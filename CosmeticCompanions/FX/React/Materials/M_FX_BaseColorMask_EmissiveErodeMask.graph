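# Translucent, unlit, two-sided Niagara material for companion reaction effects, defaulting to the viperBucket logo mask texture.
# The packed texture's R blends shade to highlight, G masks particle-tinted emissive, and the summed color is divided by eye adaptation.
# Opacity combines texture and particle alpha with a B-channel SmoothStep driven by dynamic Erode; Erode_Softness is unused, as Min subtracts Erode from itself.
# Depth fade applies except at Low and Medium quality.
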Material /CosmeticCompanions/FX/React/Materials/M_FX_BaseColorMask_EmissiveErodeMask
graph {
 properties {
  bAutomaticallySetUsageInEditor=false
  bUsedWithNiagaraMeshParticles=true
  bUsedWithNiagaraRibbons=true
  bUsedWithNiagaraSprites=true
  blendMode=BLEND_Translucent
  shadingModel=MSM_Unlit
  twoSided=true
  usedShadingModels=Unlit
 }
 nodes [
  n1 = VectorParameter() {parameterName=Tex_BaseColor_Shade, defaultValue=<rgba 0 0 0 1>}
  n2 = VectorParameter() {parameterName=Tex_BaseColor_Highlight, defaultValue=<rgba 1 1 1 1>}
  n3 = TextureSampleParameter2D() {
   parameterName=TexColor
   samplerType=SAMPLERTYPE_LinearColor
   texture=/CosmeticCompanions/FX/React/Textures/T_ReactFX_viperBucket_logo_erodeMask
  }
  n4 = ParticleColor()
  n5 = Multiply(A=VectorParameter() {parameterName=EmissiveColor, defaultValue=<rgba 0 0 0 1>}.RGB, B=@n4.RGB)
  Divide(A=Add(A=LinearInterpolate(A=@n1.RGB, B=@n2.RGB, Alpha=@n3.R), B=Multiply(A=@n5, B=@n3.G)), B=EyeAdaptation().EyeAdaptation)
  n6 = DynamicParameter() {defaultValue=<rgba 0.3192 0.05 0 0>, paramNames [Erode, Erode_Softness, "", Param4]}
  n7 = Subtract(A=@n6.Erode, B=@n6.Erode)
  n8 = Multiply(A=Multiply(A=@n3.A, B=@n4.A), B=Saturate(None=SmoothStep(Min=@n7, Max=@n6.Erode, Value=@n3.B) {constMax=0, constMin=-0.1}))
  n9 = DepthFade(Opacity=@n8, FadeDistance=ScalarParameter() {defaultValue=10, parameterName=DepthFadeDistance})
  QualitySwitch(Default=@n9, Low=@n8, High=@n9, Medium=@n8, Epic=@n9)
 ]
}
