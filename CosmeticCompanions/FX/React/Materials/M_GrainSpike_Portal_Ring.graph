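# Unlit translucent material for Niagara mesh particles, apparently the ring of a GrainSpike portal reaction effect for cosmetic companions.
# A panning fire-trail texture, distorted by scrolling wispy noise, drives opacity and particle-tinted emissive, with a depth-color glow added near intersecting geometry.
# SplineThicken takes its width from a WidthMultiplier dynamic parameter and supplies normals for an inverted Fresnel that softens edges.
# Opacity also uses depth fade and optional particle alpha.
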Material /CosmeticCompanions/FX/React/Materials/M_GrainSpike_Portal_Ring
graph {
 properties {
  bAutomaticallySetUsageInEditor=false
  bUsedWithNiagaraMeshParticles=true
  bUsedWithStaticMesh=false
  blendMode=BLEND_Translucent
  pixelDepthOffsetMode=PDOM_Legacy
  shadingModel=MSM_Unlit
  usedShadingModels=Unlit
 }
 nodes [
  n1 = VectorParameter() {parameterName=v, defaultValue=<rgba 1 1 0 1>}
  n2 = MaterialFunctionCall(XY=Multiply(A=AppendVector(A=@n1.R, B=@n1.R), B=TextureCoordinate() {vTiling=3})) {materialFunction=&Swizzle}
  n3 = ScalarParameter() {defaultValue=1, parameterName=Speed}
  n4 = Multiply(A=@n3, B=Time() {bOverride_Period=true, period=114514})
  n5 = TextureCoordinate() {uTiling=2}
  n6 = Time() {bOverride_Period=true, period=114514}
  n7 = TextureSample(UVs=Panner(Coordinate=@n5, Time=Multiply(A=@n6, B=@n3)) {speedX=0.15, speedY=-0.2}) {samplerType=SAMPLERTYPE_Grayscale, texture=/Game/Effects/Fort_Effects/Textures/Fire/T_WispyFireNoise_Soft_M}
  n8 = ScalarParameter() {defaultValue=0.15, parameterName=DistortionIntensity}
  n9 = TextureSampleParameter2D(UVs=Add(A=Panner(Coordinate=@n2.YX, Time=@n4) {speedX=-0.3}, B=Multiply(A=@n7.R, B=@n8))) {
   parameterName=MainTexture
   samplerType=SAMPLERTYPE_LinearGrayscale
   texture=/BRCosmetics/Athena/Effects/Particles/Skydiving/NS_Contrail_LightCycle/Textures/T_FireTrail_01
  }
  n10 = Multiply(A=ScalarParameter() {defaultValue=7.5}, B=DynamicParameter() {paramNames [WidthMultiplier, Param2, Param3, Param4]}.WidthMultiplier)
  n11 = MaterialFunctionCall(WidthBase=@n10, WidthTip=@n10, Expand U or V UV Channel=StaticBool() {value=true}) {materialFunction=&SplineThicken}
  n12 = Multiply(
   A=Multiply(A=ScalarParameter() {defaultValue=2, parameterName=Param_1}, B=SmoothStep(Value=@n9.R) {constMax=0.9}) {constA=1}
   B=OneMinus(None=Fresnel(Normal=@n11.Normal) {baseReflectFraction=0, exponent=2})
  )
  n13 = ParticleColor()
  DepthFade(Opacity=LinearInterpolate(A=@n12, B=Multiply(A=@n13.A, B=@n12), Alpha=ScalarParameter() {parameterName=UseParticleAlpha})) {fadeDistanceDefault=1}
  MaterialFunctionCall(XY=@n5) {materialFunction=&Swizzle}
  n14 = OneMinus(None=DepthFade() {fadeDistanceDefault=0.5})
  n15 = VectorParameter() {parameterName=DepthColor, defaultValue=<rgba 3 1.425 0.3 0>}
  n16 = Multiply(A=VectorParameter() {defaultValue=<rgba 0.45 0.1222 0.0225 1>}.RGB, B=@n13.RGB)
  n17 = SmoothStep(Value=@n9.R) {constMax=0.3, constValue=0.6}
  MaterialFunctionCall(Emissive=Add(A=Multiply(A=@n14, B=@n15.RGB), B=Multiply(A=@n16, B=@n17))) {materialFunction=&MF_FX_EmissiveAdjustments}
  VertexColor()
 ]
}
