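# Material function, exposed to the library, that converts a linear ramp into an inverse-distance falloff with a hot core, stored with the DrapeUltra back bling FX.
# It divides CoreIntensity by one minus InRamp and subtracts CoreIntensity, so output starts at zero and climbs steeply as the ramp nears one.
# DetailBalence sets the exponent on that curve, and MaxOutPut clamps the spike.
# A soft UV SphereMask serves as InRamp's preview default.
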MaterialFunction /BRCosmetics/Accessories/FORT_Backpacks/Backpack_F_MED_DrapeUltra/FX/Materials/MF_MOON_InvertDistanceRamp
graph {
 properties {bExposeToLibrary=true}
 nodes [
  n1 = FunctionInput() {
   bUsePreviewValueAsDefault=true
   description="DefaultValue = 0.05"
   inputName=CoreIntensity
   inputType=FunctionInput_Scalar
   sortPriority=2
   previewValue=<0.05 0 0 1>
  }
  n2 = Constant() {r=0.5}
  n3 = FunctionInput(Preview=SphereMask(A=TextureCoordinate(), B=@n2, Radius=@n2) {hardnessPercent=0}) {bUsePreviewValueAsDefault=true, inputName=InRamp, inputType=FunctionInput_Scalar, sortPriority=1}
  n4 = Subtract(A=Divide(A=@n1, B=OneMinus(None=@n3)), B=@n1)
  n5 = FunctionInput() {
   bUsePreviewValueAsDefault=true
   description="DefaultValue = 1.0"
   inputName=DetailBalence
   inputType=FunctionInput_Scalar
   sortPriority=3
   previewValue=<1 0 0 1>
  }
  n6 = FunctionInput() {bUsePreviewValueAsDefault=true, inputName=MaxOutPut, inputType=FunctionInput_Scalar, sortPriority=4, previewValue=<1000 0 0 1>}
  FunctionOutput(None=Min(A=Power(Base=@n4, Exp=@n5), B=@n6) {constB=1000})
 ]
}
