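# Translucent, two-sided unlit parent material for the GrainSpike portal effect in Cosmetic Companions reactions, used on Niagara mesh particles.
# Layered sphere masks with BumpOffset parallax give the portal apparent depth and drive both opacity and a color-curve atlas ramp.
# Radially mapped, panning wispy noise supplies UV distortion and texture detail.
# Emissive adds particle-tinted ramp color to a depth-fade intersection glow, while opacity is softened by depth fades.
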Material /CosmeticCompanions/FX/React/Materials/M_GrainSpike_Portal_parent
graph {
 properties {
  bAutomaticallySetUsageInEditor=false
  bUsedWithNanite=true
  bUsedWithNiagaraMeshParticles=true
  bUsedWithStaticMesh=false
  blendMode=BLEND_Translucent
  pixelDepthOffsetMode=PDOM_Legacy
  shadingModel=MSM_Unlit
  twoSided=true
  usedShadingModels=Unlit
 }
 nodes [
  MaterialFunctionCall(Emissive=NamedRerouteUsage()) {materialFunction=&MF_FX_EmissiveAdjustments}
  n1 = SphereMask(
   A=TextureCoordinate()
   B=Constant() {r=0.5}
   Radius=ScalarParameter() {defaultValue=0.5, group=Sphere Mask, parameterName=Sphere Mask Radius}
   Hardness=NamedRerouteUsage()
  ) {hardnessPercent=50}
  n2 = NamedRerouteUsage()
  n3 = SphereMask(A=@n2, B=Constant() {r=0.5}, Radius=ScalarParameter() {defaultValue=1, parameterName=Bump Sphere Mask Radius}, Hardness=NamedRerouteUsage()) {hardnessPercent=50}
  n4 = SphereMask(
   A=@n2
   B=Constant() {r=0.5}
   Radius=ScalarParameter() {defaultValue=2, parameterName=Bump Sphere Mask 2 Radius}
   Hardness=ScalarParameter() {parameterName=Bump Sphere Mask 2 Hardness}
  ) {hardnessPercent=50}
  n5 = Multiply(
   A=Max(A=OneMinus(None=@n1), B=OneMinus(None=@n3))
   B=LinearInterpolate(B=OneMinus(None=@n4), Alpha=NamedRerouteUsage()) {constA=1, constAlpha=1}
  )
  n6 = MaterialFunctionCall!(
   InRamp=@n5
   CoreIntensity=ScalarParameter() {defaultValue=2.5, parameterName=OverAllIntensity}
   DetailBalence=ScalarParameter() {defaultValue=1.5, parameterName=Param_6}
   MaxOutPut=Constant() {r=10}
  ) {materialFunction=/BRCosmetics/Accessories/FORT_Backpacks/Backpack_F_MED_DrapeUltra/FX/Materials/MF_MOON_InvertDistanceRamp}
  Multiply(A=@n6.Result, B=NamedRerouteUsage())
  n7 = Add(A=BumpOffset(Coordinate=TextureCoordinate(), Height=ScalarParameter() {defaultValue=-1.25, parameterName=Param_1}) {referencePlane=0}) {constB=-0.5}
  n8 = VectorParameter() {parameterName=Distortion_UV, defaultValue=<rgba 1 3 -0.1 0.5>}
  n9 = TextureSample!(
   UVs=MaterialFunctionCall!(
    ClampTexture
    UV=MaterialFunctionCall(Vector or UVs=@n7) {materialFunction=&VectorToRadialValue}.Radial Coordinates
    Tex
    InRGB=@n8.RGB
    InA=@n8.RGB
    Time
    TimeMult
   ) {materialFunction=/BRCosmetics/Accessories/FORT_Backpacks/Backpack_F_MED_DrapeUltra/FX/Materials/MF_MOON_TexPanner_V2}.PannerUV
   Tex
   Level=Constant() {r=1.5}
   Apply View MipBias
  ) {mipValueMode=TMVM_MipLevel, samplerType=SAMPLERTYPE_Grayscale, texture=/Game/Effects/Fort_Effects/Textures/Fire/T_WispyFireNoise_Soft_M}
  NamedRerouteDeclaration(None=Multiply(A=@n9.R, B=ScalarParameter() {defaultValue=0.02, parameterName=DistortionIntensity})) {name=UVDistortion, nodeColor=<rgba 0 1 0.9063 1>}
  NamedRerouteDeclaration(None=NamedRerouteUsage()) {name=OUT_EdgeMask, nodeColor=<rgba 0.125 1 0 1>}
  n10 = Multiply(
   A=Power(Base=@n5, Exp=ScalarParameter() {defaultValue=2, parameterName=Sphere Mask Power})
   B=ScalarParameter() {defaultValue=2, parameterName=Sphere Mask Strenth}
  )
  n11 = NamedRerouteUsage()
  n12 = Multiply(A=Power(Base=OneMinus(None=@n10), Exp=LinearInterpolate(Alpha=@n11) {constA=100}) {constExponent=1}, B=NamedRerouteUsage())
  n13 = Multiply(
   A=DepthFade(Opacity=Saturate(None=@n12), FadeDistance=ScalarParameter() {parameterName=DepthFadeDistance})
   B=ScalarParameter() {defaultValue=5, parameterName=FinalOpacity}
  ) {constB=5}
  NamedRerouteDeclaration(None=DepthFade(Opacity=Saturate(None=@n13)) {fadeDistanceDefault=1}) {name=OUT_Transparent, nodeColor=<rgba 1 0.8438 0 1>}
  NamedRerouteDeclaration(None=BumpOffset(Coordinate=TextureCoordinate(), Height=ScalarParameter() {defaultValue=-1.25, parameterName=Param_1}) {referencePlane=0}) {name=BumpUV, nodeColor=<rgba 0.6406 0 1 1>}
  n14 = Constant() {r=0.5}
  n15 = SphereMask(A=TextureCoordinate(), B=@n14, Radius=@n14) {attenuationRadius=0.5, hardnessPercent=75}
  NamedRerouteDeclaration(None=@n15) {name=OUT_SphereMask, nodeColor=<rgba 1 0.8672 0 1>}
  NamedRerouteDeclaration(None=ParticleColor().A) {name=OUT_ParticleAlpha, nodeColor=<rgba 0.6172 1 0 1>}
  NamedRerouteDeclaration(None=LinearInterpolate(A=Constant() {r=0.7}, Alpha=NamedRerouteUsage()) {constB=0}) {name=OUT_Dissolve, nodeColor=<rgba 0.6875 1 0 1>}
  n16 = ScalarParameter() {defaultValue=0.6, parameterName=Emissive Color Max}
  n17 = ScalarParameter() {defaultValue=-1, parameterName=Emissive Color Min}
  n18 = MaterialFunctionCall(Float2=Add(A=NamedRerouteUsage(), B=NamedRerouteUsage())) {materialFunction=&BreakOutFloat2Components}
  n19 = Add(A=@n18.G, B=Multiply(A=@n18.R, B=ScalarParameter() {defaultValue=1, parameterName=UV_Tilt}))
  n20 = TextureSampleParameter2D!(
   UVs=NamedRerouteDeclaration(None=MaterialFunctionCall(X=@n18.R, Y=@n19) {materialFunction=&MakeFloat2}.Result) {name=TiltedUV, nodeColor=<rgba 0 0.3848 1 1>}
   Level=ScalarParameter() {defaultValue=1.5, parameterName=Mip Level}
   Apply View MipBias
  ) {
   mipValueMode=TMVM_MipLevel
   parameterName=Main Emissive TX
   texture=/BRCosmetics/Athena/Effects/Particles/Skydiving/NS_Contrail_Chainmail/T_WispyFireNoise
  }
  n21 = SquareRoot(None=@n20.G)
  n22 = NamedRerouteUsage()
  NamedRerouteDeclaration(None=Subtract(A=LinearInterpolate(A=@n16, B=@n17, Alpha=@n21), B=OneMinus(None=@n22))) {name=OUT_TextureLerp, nodeColor=<rgba 0.9688 0 1 1>}
  n23 = Add(A=NamedRerouteUsage()) {constB=-0.5}
  n24 = VectorParameter() {parameterName=Main_UV, defaultValue=<rgba 1 2 -0.2 0.25>}
  NamedRerouteDeclaration(None=MaterialFunctionCall(
    UV=MaterialFunctionCall(Vector or UVs=@n23) {materialFunction=&VectorToRadialValue}.Radial Coordinates
    InRGB=@n24.RGB
    InA=@n24.RGB
   ) {materialFunction=&MF_MOON_TexPanner_V2}.PannerUV) {name=MainUV, nodeColor=<rgba 0 0.4049 1 1>}
  n25 = OneMinus(None=DepthFade() {fadeDistanceDefault=0.5})
  n26 = VectorParameter() {parameterName=DepthColor, defaultValue=<rgba 0 1.8963 3 1>}
  n27 = ParticleColor()
  n28 = CurveAtlasRowParameter(CurveTime=@n6.Result) {
   atlas=/CosmeticCompanions/FX/React/Materials/ColorCurves/CA_GrainSpike_Portals
   curve=/CosmeticCompanions/FX/React/Materials/ColorCurves/C_GrainSpike_Portal_2
  }
  NamedRerouteDeclaration(None=Add(A=Multiply(A=@n25, B=@n26.RGB), B=Multiply(A=@n27.RGB, B=@n28))) {name=OUT_Emissive, nodeColor=<rgba 1 0 0.3047 1>}
  NamedRerouteUsage()
  NamedRerouteUsage()
  OneMinus(None=@n15)
 ]
}
