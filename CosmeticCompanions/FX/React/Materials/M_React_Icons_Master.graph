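# Translucent unlit master material for Niagara sprite icons, likely companion reaction effects.
# A sprite texture, scaled about its center by Icon_UVs and dynamic XScale, passes through SubUV, where SubUV_UCount drives both counts and time, leaving VCount, Frame and YScale unused.
# Particle color tints it, with an optional time-of-day collection tint before shared emissive adjustments.
# A switch takes opacity from texture red or alpha, times particle alpha.
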Material /CosmeticCompanions/FX/React/Materials/M_React_Icons_Master
graph {
 properties {
  bAutomaticallySetUsageInEditor=false
  bUsedWithNiagaraSprites=true
  blendMode=BLEND_Translucent
  shadingModel=MSM_Unlit
  usedShadingModels=Unlit
 }
 nodes [
  Constant() {r=2}
  n1 = CollectionParameter() {
   collection=&FortniteMaterialParameters
   group=ToD
   parameterName=FXTimeOfDayColorTint
  }
  n2 = LinearInterpolate(
   A=Constant3Vector() {constant=<rgba 1 1 1 0>}
   B=ComponentMask(None=@n1) {b=true, g=true, r=true}
   Alpha=ScalarParameter() {defaultValue=1, group=ToD, parameterName=ToD Contribution, sortPriority=33}
  )
  n3 = ParticleColor()
  n4 = DynamicParameter() {parameterIndex=1, paramNames [SubUV_UCount, SubUV_VCount, SubUV_Frame, ""]}
  n5 = TextureCoordinate()
  n6 = DynamicParameter() {paramNames [XScale, YScale, "", ""]}
  n7 = Multiply(
   A=ComponentMask(None=VectorParameter() {parameterName=Icon_UVs, defaultValue=<rgba 1 1 0 1>}.RGB) {g=true, r=true}
   B=AppendVector(A=@n6.XScale, B=@n6.XScale)
  )
  n8 = TextureSampleParameter2D(UVs=MaterialFunctionCall!(
    U_Images=@n4.SubUV_UCount
    V_Images=@n4.SubUV_UCount
    SubUV Texture RGB
    Time Speed
    Manual Time Control=StaticBool() {value=true}
    Time=@n4.SubUV_UCount
    UVs=MaterialFunctionCall(UVs=@n5, Texture Scale=@n7) {materialFunction=&ScaleUVsByCenter}.UVs
   ) {materialFunction=/Game/Effects/Fort_Effects/Materials/_Functions/MF_SubUV}.UVs) {
   group=Texture
   parameterName=Sprite
   samplerSource=SSM_Clamp_WorldGroupSettings
   texture=/Game/Effects/Fort_Effects/Textures/Sparks/T_Sparkles_4Star_01_D
  }
  n9 = Multiply(A=@n3.RGB, B=@n8.RGB)
  MaterialFunctionCall(Emissive=StaticSwitchParameter(True=Multiply(A=@n2, B=@n9), False=@n9) {group=ToD, parameterName=Use FXTimeOfDayColorTint MPC, sortPriority=31}) {materialFunction=&MF_FX_EmissiveAdjustments}
  StaticSwitchParameter(True=Multiply(A=@n3.A, B=@n8.R), False=Multiply(A=@n3.A, B=@n8.A)) {group=Texture, parameterName=Use Texture R As Alpha, sortPriority=33}
 ]
}
