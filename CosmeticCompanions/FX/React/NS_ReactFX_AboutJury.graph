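# Niagara reaction effect for a cosmetic companion that plays a shared petting-reaction sound and an animated pop.
# An Intro sprite squashes and stretches while moving from the system's heart spawn point to its burst position.
# GPU emitters time bubble/heart, crack and spike sprite bursts to the intro's end on camera-oriented shapes, alongside a CPU mesh-and-sprite glow.
# Shared system values drive timing and scale, with time-of-day tinting and front-end dimming.
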NiagaraSystem /CosmeticCompanions/FX/React/NS_ReactFX_AboutJury
dependencies {
 usedDataInterfaces [
  &NiagaraDataInterfaceAudioPlayer
  &NiagaraDataInterfaceCurve
  &NiagaraDataInterfaceVector2DCurve
  &NiagaraDataInterfaceCamera
  &NiagaraDataInterfaceSpriteRendererInfo
  &NiagaraDataInterfaceArrayInt32
  &NiagaraDataInterfacePlatformSet
  &NiagaraDataInterfaceVectorCurve
  &NiagaraDataInterfaceColorCurve
 ]
 usedDynamicInputs [
  &FN_FrontEndVector3Modifier
  &ColorFromCurve
  &MultiplyLinearColor
  &MakeVector4_FromLinearColor
  &Vector2DFromFloat
 ]
 usedModules [
  &self:SystemSpawnScript.SystemScriptSource.SystemScriptGraph.NiagaraNodeAssignment_11.SetVariables_#1
  &self:SystemSpawnScript.SystemScriptSource.SystemScriptGraph.NiagaraNodeAssignment_12.SetVariables_#2
  &self:SystemSpawnScript.SystemScriptSource.SystemScriptGraph.NiagaraNodeAssignment_13.SetVariables_#3
  &self:SystemSpawnScript.SystemScriptSource.SystemScriptGraph.NiagaraNodeAssignment_14.SetVariables_#4
  &self:SystemSpawnScript.SystemScriptSource.SystemScriptGraph.NiagaraNodeAssignment_15.SetVariables_#5
  &self:SystemSpawnScript.SystemScriptSource.SystemScriptGraph.NiagaraNodeAssignment_16.SetVariables_#6
  &self:SystemSpawnScript.SystemScriptSource.SystemScriptGraph.NiagaraNodeAssignment_17.SetVariables_#7
  &SystemState
 ]
}
emitters {
 BubblesHearts {
  simTarget=GPUComputeSim
  emitterUpdateScript [
   &EmitterState(Life Cycle Mode=::NewEnumerator1, Loop Behavior=::NewEnumerator1, Loop Duration=1.05, UseLoopDelay=false)
   &SpawnBurst_Instantaneous(
    Spawn Count=&RandomRangeInt(Minimum=6, Maximum=7)
    Spawn Time=&Add_Float(A=$System.IntroLength, B=-0.1)
    Use Spawn Probability=false
    UseLoopCountLimit=false
   )
  ]
  particleSpawnScript [
   /SetVariables(Particles.VisibilityTag=$Emitter.SpawnGroup) {isSetParameters=true}
   &InitializeParticle(
    Color=<rgba 0.402 0.1441 0.0296 1>
    Lifetime=&RandomRangeFloat(Minimum=0.6, Maximum=1)
    Position Offset=<0 0 100>
    Sprite Size Mode=::NewEnumerator3
    Uniform Sprite Size=&FloatFromCurve(
     FloatCurve=DataInterface() {propertyValues {Curve {defaultValue=1, keys [{tangentMode=RCTM_User, time=0, value=5}, {tangentMode=RCTM_User, time=1, value=25}]}}}
     CurveIndex=&ReturnNormalizedExecIndex()
    )
    UsePositionOffset=false
    Write Material Random=true
   )
   &OffsetPosition(Position Offset=&Lerp_Vector(
     A=$System.HeartIntroSpawnPosition
     B=$System.BurstPosition
     Alpha=&FloatFromCurve(
      FloatCurve=DataInterface() {propertyValues {Curve {keys [
          {arriveTangent=3.0606, interpMode=RCIM_Cubic, leaveTangent=3.0606, tangentMode=RCTM_User, time=0, value=0}
          {arriveTangent=-0.1073, tangentMode=RCTM_User, time=1, value=1}
         ]}}}
      CurveIndex=&Divide_Float(A=$Emitter.Age, B=$System.IntroLength)
     )
    ))
   &ShapeLocation(
    MaskSpawnGroup=false
    Non Uniform Scale=&VectorFromFloat(Value=$System.GlobalScale)
    Ring / Disc Distribution Mode:ENiagaraSphereDistributionMode=::NewEnumerator3
    Ring Radius=&RandomRangeFloat(Minimum=15, Maximum=20)
    Rotation Axis=&NormalizeVector(
     Vector To Normalize=&Multiply_VectorByFloat(Vector=&ReturnCameraVector(Camera Vectors=::NewEnumerator1, Coordinate Space=::Simulation), Float=1)
     Safe Normalize=true
     Safe Fallback Threshold=0
     Safe Fallback Vector=<0 0 0.001>
    )
    Rotation Mode=::NewEnumerator1
    Shape Primitive=::NewEnumerator4
    UseOwnerScale=true
   )
   &AddVelocity(
    Module.Use Falloff Distance=false
    Origin Offset=$System.BurstPosition
    Override Seed=false
    Velocity Mode=::NewEnumerator1
    Velocity Speed=&RandomRangeFloat(Minimum=100, Maximum=150)
   )
   /SetVariables(Particles.float:NiagaraFloat=&RandomRangeFloat(Minimum=3, Maximum=6)) {isSetParameters=true}
   &SpriteFacingAndAlignment(
    Sprite Alignment=&NormalizeVector(Vector To Normalize=$Particles.Velocity, Safe Normalize=true, Safe Fallback Threshold=0, Safe Fallback Vector=<0 0 0.001>)
    WriteAlignment=true
    WriteFacing=false
   )
  ]
  particleUpdateScript [
   &ParticleState()
   &ScaleSpriteSize(
    Scale Sprite Size Mode=::NewEnumerator0
    Uniform Scale Factor=&FloatFromCurve(
     FloatCurve=DataInterface() {propertyValues {Curve {keys [
         {arriveTangent=13.8261, interpMode=RCIM_Cubic, leaveTangent=13.8261, tangentMode=RCTM_User, time=0, value=0}
         <at 0.2487 1.1246 RCIM_Cubic>
         {arriveTangent=-2.6397, interpMode=RCIM_Cubic, leaveTangent=-2.6397, tangentMode=RCTM_User, time=1, value=0}
        ]}}}
     Scale Curve=$System.GlobalScale
    )
   )
   &DynamicMaterialParameters(
    Index 0 Param 1=1
    Index 0 Param 2=1
    Module.Param0WriteEnabled=true
    Module.Param1WriteEnabled=true
    Module.Param2WriteEnabled=true
    Module.Param3WriteEnabled=true
   )
   &Drag(Drag=7.5, Use Linear Drag=true, Use Rotational Drag=true)
   &SolveForcesAndVelocity(Clamp Velocity=false, Limit Acceleration=false)
   &ScaleColor(
    Scale Alpha=&MergeHelper_FloatSwitch(
     0=&FloatFromCurve(FloatCurve=DataInterface() {propertyValues {Curve {
         defaultValue=0
         keys [
          {arriveTangent=6.566, interpMode=RCIM_Cubic, leaveTangent=6.566, tangentMode=RCTM_User, time=0.95, value=1}
          {arriveTangent=35.933, interpMode=RCIM_Cubic, leaveTangent=35.933, tangentMode=RCTM_User, time=1, value=0}
         ]
        }}})
     1=1
     2=2
     3=?
     Index=&SpawnGroup()
     Num=::NewEnumerator1
    )
    Scale RGB=&MakeVectorFromLinearColorRGB(Color=<rgba 1 1 1 1>)
    ScaleA=true
    ScaleRGB=true
   )
   /SetVariables(Particles.Glow Color=<rgba 0 0 0 1>) {isSetParameters=true}
   /SetVariables(Particles.Glow Size=&Multiply_Vector2D_ByFloat(
     Float=&Multiply_Float(
      A=&Lerp_Float(A=1.175, B=1.08, Alpha=&ReturnNormalizedExecIndex())
      B=&FloatFromCurve(FloatCurve=DataInterface() {propertyValues {Curve {defaultValue=1, keys [{tangentMode=RCTM_User, time=0, value=1}, {tangentMode=RCTM_User, time=1, value=1.15}]}}})
     )
     Vector2D=$Particles.SpriteSize
    )) {isSetParameters=true}
   /SetVariables(Particles.DMP_Lion_00:Vector4f=<0 1 0 0>) {isSetParameters=true}
  ]
  rendererClasses [&NiagaraSpriteRendererProperties]
  renderers [
   {rendererClass=&NiagaraSpriteRendererProperties}
   {rendererClass=&NiagaraSpriteRendererProperties, rendererIndex=1}
  ]
 }
 CrackGraphic {
  simTarget=GPUComputeSim
  emitterUpdateScript [
   &EmitterState(Life Cycle Mode=::NewEnumerator1, Loop Behavior=::NewEnumerator1, Loop Duration=1, UseLoopDelay=false)
   &SpawnBurst_Instantaneous(
    Spawn Count=&RandomRangeInt(Minimum=5, Maximum=6)
    Spawn Time=&Add_Float(A=$System.IntroLength, B=-0.1)
    Use Spawn Probability=false
    UseLoopCountLimit=true
   )
  ]
  particleSpawnScript [
   &PartitionParticles(Number Of Partitions=3)
   &InitializeParticle(
    Color=<rgba 2 1.0666 0 1>
    Lifetime=&RandomRangeFloat(Minimum=0.3, Maximum=0.5)
    Sprite Size=<35 80>
    Sprite Size Mode=::NewEnumerator4
    UsePositionOffset=false
    Write Material Random=true
   )
   &SubUVAnimation(
    Evaluation Type=::NewEnumerator0
    "Sprite Renderer "=DataInterface() {propertyValues {SpriteRenderer=&self:CrackGraphic_0.NiagaraSpriteRendererProperties_3}}
    SubUV Animation Mode:ENiagara_SubUVLookupModeV2=::NewEnumerator2
    UseEndFrame=false
    UseStartFrame=false
   )
   &ShapeLocation(
    Handle Radius=0
    Large Radius=20
    MaskSpawnGroup=false
    Non Uniform Scale=&VectorFromFloat(Value=$System.GlobalScale)
    Offset=$System.BurstPosition
    Offset Mode=::NewEnumerator0
    Rotation Angle:NiagaraFloat=&RandomRangeFloat(Minimum=0, Maximum=45)
    Rotation Angle Type:ENiagara_AngleInput=::NewEnumerator0
    Rotation Axis=&ReturnCameraVector(Camera Vectors=::NewEnumerator1, Coordinate Space=::Local)
    Rotation Mode=::NewEnumerator1
    Shape Primitive=::NewEnumerator3
    Torus Distribution Mode:ENiagaraTorusDistributionMode=::NewEnumerator1
    U Position=&ReturnNormalizedExecIndex(Normalized Index Includes One=false)
    UseOwnerScale=true
    V Position=0
   )
   &AddVelocity(
    Velocity=$Particles.ShapeLocation.ShapeNormal
    Velocity Speed Scale=&RandomRangeFloat(Minimum=50, Maximum=150)
   )
  ]
  particleUpdateScript [
   &ParticleState()
   &CameraOffset(Camera Offset Amount=-60)
   /SetVariables(Particles.SpriteScale:NiagaraFloat=&FloatFromCurve(FloatCurve=DataInterface() {propertyValues {Curve {keys [
         {interpMode=RCIM_Cubic, tangentMode=RCTM_User, time=0, value=0}
         {interpMode=RCIM_Cubic, tangentMode=RCTM_User, time=0.1, value=1}
         {arriveTangent=-2.6589, tangentMode=RCTM_User, time=1, value=0}
        ]}}})) {isSetParameters=true}
   /SetVariables(Particles.SpriteSize=&Multiply_Vector2D_ByFloat(Float=$Particles.SpriteScale, Vector2D=$Particles.Initial.SpriteSize)) {isSetParameters=true}
   &ScaleSpriteSize(
    Uniform Curve Scale=$System.GlobalScale
    Uniform Curve Sprite Scale=DataInterface() {propertyValues {Curve {keys [
        {interpMode=RCIM_Cubic, tangentMode=RCTM_User, time=0, value=0}
        {interpMode=RCIM_Cubic, tangentMode=RCTM_User, time=0.05, value=1}
        {arriveTangent=-2.6589, tangentMode=RCTM_User, time=1, value=0}
       ]}}}
   )
   &ScaleSpriteSize(
    Scale Sprite Size Mode=::NewEnumerator0
    Uniform Scale Factor=&FloatFromCurve(FloatCurve=DataInterface() {propertyValues {Curve {keys [<at 0 1>,<at 0.1 0.6>,<at 0.3 1>,<at 0.5 0.6>,<at 0.7 1>,<at 0.9 0.6>,<at 1 1>]}}})
   ) {module=ScaleSpriteSize001}
   &ScaleSpriteSize(
    Non-Uniform Curve Sprite Scale=DataInterface() {propertyValues {
      XCurve {keys [{arriveTangent=2.4297, interpMode=RCIM_Cubic, leaveTangent=2.4297, tangentMode=RCTM_User, time=0, value=0}, <at 1 1 RCIM_Cubic>]}
      YCurve {keys [
        {arriveTangent=5.7604, interpMode=RCIM_Cubic, leaveTangent=5.7604, tangentMode=RCTM_User, time=0, value=0}
        {arriveTangent=3.3333, interpMode=RCIM_Cubic, leaveTangent=3.3333, time=0.2, value=1.1}
        {interpMode=RCIM_Cubic, leaveTangent=2.5625, time=0.3, value=1}
       ]}
     }}
    Scale Sprite Size Mode=::NewEnumerator3
   ) {module=ScaleSpriteSize002}
   &SpriteRotationRate(Rotation Rate=&Multiply_FloatByInt(
     Float=&FloatFromCurve(
      FloatCurve=DataInterface() {propertyValues {Curve {keys [<at 0 1 RCIM_Cubic>, {arriveTangent=-2.1844, interpMode=RCIM_Cubic, leaveTangent=-2.1844, tangentMode=RCTM_User, time=1, value=0}]}}}
      Scale Curve=&Multiply_Float(A=5000, B=&RandomRangeFloat(Minimum=0.6, Maximum=1.2))
     )
     Integer=&SelectIntFromArray(
      Array Sampling Mode=::NewEnumerator1
      Direct Array Index=?
      Int Selection Array=DataInterface() {propertyValues {IntData [1,-1]}}
      Random Seed=0
     )
    ))
   &Drag(Drag=10, Use Linear Drag=true, Use Rotational Drag=true) {module=Drag001}
   &SolveForcesAndVelocity(Clamp Velocity=false, Limit Acceleration=false)
   &ScaleColor(
    Scale RGB=&MakeVectorFromLinearColorRGB(Color=&FNTimeOfDayColor(
      Night Color=<rgba 0.7152 0.6979 1 1>
      Morning Color=<rgba 0.9922 0.7751 0.8928 1>
      Day Color=<rgba 1 1 1 1>
      Evening Color=<rgba 0.9922 0.5725 0.8 1>
     ))
    ScaleA=true
    ScaleRGB=false
   ) {module=ScaleColor001}
   /SetVariables(Particles.Glow Size=&Multiply_Vector2D_ByFloat(Float=1.15, Vector2D=$Particles.SpriteSize)) {isSetParameters=true}
   /SetVariables(Particles.Glow Color=<rgba 0 0 0 1>) {isSetParameters=true}
  ]
  rendererClasses [&NiagaraSpriteRendererProperties]
  renderers [
   {rendererClass=&NiagaraSpriteRendererProperties}
   {rendererClass=&NiagaraSpriteRendererProperties, rendererIndex=1}
  ]
 }
 Glow {
  emitterUpdateScript [
   &EmitterState(Inactive Response=::NewEnumerator1, Life Cycle Mode=::NewEnumerator1, Loop Behavior=::NewEnumerator1, Loop Duration=2, UseLoopDelay=false)
   &SpawnBurst_Instantaneous(Spawn Count=1, Spawn Group=1, Spawn Time=0.3, Use Spawn Probability=false) {module=SpawnBurst_Instantaneous001}
   &SpawnBurst_Instantaneous(Spawn Time=0.075, Use Spawn Probability=false, UseLoopCountLimit=false)
  ]
  particleSpawnScript [
   /SetVariables(Particles.VisibilityTag=&SpawnGroup()) {isSetParameters=true}
   &InitializeParticle(
    Color=&MergeHelper_ColorSwitch(0=<rgba 2 1.02 0.04 10>, 1=<rgba 1 0.5333 0 1>, 2=?, 3=?, Index=$Particles.VisibilityTag, Num=::NewEnumerator0)
    Lifetime=&MergeHelper_FloatSwitch(0=1.25, 1=1, 2=?, 3=?, Index=$Particles.VisibilityTag, Num=::NewEnumerator0)
    Mesh Scale=<0.5 0.5 0.5>
    Mesh Scale Mode=::NewEnumerator4
    Position Offset=&AddVectorToPosition(
     Position=&MergeHelper_VectorSwitch(0=<0 0 -40>, 1=<0 0 0>, 2=?, 3=?, Index=$Particles.VisibilityTag, Num=::NewEnumerator0)
     Vector=$System.BurstPosition
    )
    Sprite Rotation Mode=::NewEnumerator2
    Sprite Size Mode=::NewEnumerator3
    Uniform Sprite Size=64.8455
    UsePositionOffset=true
    Write Material Random=true
   )
   &CameraOffset(Camera Offset Amount=&MergeHelper_FloatSwitch(
     0=&FN_PickFloatDependOnPlatform(TargetPlatformActive=-100, TargetPlatformNoActive=-150)
     1=4
     2=?
     3=?
     Index=$Particles.VisibilityTag
     Num=::NewEnumerator0
    ))
   &InitialMeshOrientation(Mesh Orientation Mode=::NewEnumerator2, Use Rotation Vector=false)
   &DynamicMaterialParameters(
    Index 0 Param 1=&MergeHelper_FloatSwitch(0=0.1, 1=50, 2=?, 3=?, Index=$Particles.VisibilityTag, Num=::NewEnumerator0)
    Index 0 Param 2=1
    Index 0 Param 3=1
    Index 0 Param 4=1
    Module.Param0WriteEnabled=true
    Module.Param1WriteEnabled=true
    Module.Param2WriteEnabled=true
    Module.Param3WriteEnabled=true
   )
  ]
  particleUpdateScript [
   &ParticleState()
   &SolveForcesAndVelocity(Clamp Velocity=false, Limit Acceleration=false)
   &ScaleColor(
    Scale Alpha=&Multiply_Float(
     A=&FN_FrontEndFloatModifier(Gameplay=1, Front End=0.5)
     B=&MergeHelper_FloatSwitch(
      0=&FloatFromCurve(
       FloatCurve=DataInterface() {propertyValues {Curve {keys [
           {arriveTangent=6.8488, interpMode=RCIM_Cubic, leaveTangent=6.8488, tangentMode=RCTM_User, time=0, value=0}
           {arriveTangent=1.25, time=0.15, value=1}
           {arriveTangent=-0.9805, interpMode=RCIM_Cubic, leaveTangent=-0.9805, tangentMode=RCTM_User, time=0.5, value=1}
           {arriveTangent=-2, interpMode=RCIM_Cubic, leaveTangent=-2, time=0.7069, value=0.3516}
           <at 1 0 RCIM_Cubic>
          ]}}}
       Scale Curve=&FNTimeOfDayScalar(Night Scale=0.35, Morning Scale=0.5, Day Scale=0.75, Evening Scale=0.5)
      )
      1=&FloatFromCurve(FloatCurve=DataInterface() {propertyValues {Curve {
          defaultValue=1
          keys [
           {arriveTangent=0.1733, interpMode=RCIM_Cubic, leaveTangent=0.1733, tangentMode=RCTM_User, time=0, value=0}
           {arriveTangent=0.5769, interpMode=RCIM_Cubic, leaveTangent=0.5769, tangentMode=RCTM_User, time=0.5352, value=0.2366}
           {arriveTangent=-4.1645, interpMode=RCIM_Cubic, leaveTangent=-4.1645, tangentMode=RCTM_User, time=0.5486, value=0.9908}
           {arriveTangent=-0.449, interpMode=RCIM_Cubic, leaveTangent=-0.449, tangentMode=RCTM_User, time=1, value=0}
          ]
         }}})
      2=?
      3=?
      Index=$Particles.VisibilityTag
      Num=::NewEnumerator0
     )
    )
    Scale RGB=&MergeHelper_VectorSwitch(
     0=&VectorFromCurve(
      VectorCurve=DataInterface() {propertyValues {
        XCurve {
         defaultValue=1
         keys [
          {arriveTangent=9.0725, interpMode=RCIM_Cubic, leaveTangent=9.0725, tangentMode=RCTM_User, time=0, value=0.25}
          {arriveTangent=-2.5249, interpMode=RCIM_Cubic, leaveTangent=-2.5249, tangentMode=RCTM_User, time=0.1408, value=1.25}
          {arriveTangent=-0.0405, interpMode=RCIM_Cubic, leaveTangent=-0.0405, tangentMode=RCTM_User, time=1, value=0.25}
         ]
        }
        YCurve {
         defaultValue=1
         keys [
          {arriveTangent=9.0725, interpMode=RCIM_Cubic, leaveTangent=9.0725, tangentMode=RCTM_User, time=0, value=0.25}
          {arriveTangent=-2.5249, interpMode=RCIM_Cubic, leaveTangent=-2.5249, tangentMode=RCTM_User, time=0.1408, value=1.25}
          {arriveTangent=-0.0405, interpMode=RCIM_Cubic, leaveTangent=-0.0405, tangentMode=RCTM_User, time=1, value=0.25}
         ]
        }
        ZCurve {
         defaultValue=1
         keys [
          {arriveTangent=9.0725, interpMode=RCIM_Cubic, leaveTangent=9.0725, tangentMode=RCTM_User, time=0, value=0.25}
          {arriveTangent=-2.5249, interpMode=RCIM_Cubic, leaveTangent=-2.5249, tangentMode=RCTM_User, time=0.1408, value=1.25}
          {arriveTangent=-0.0405, interpMode=RCIM_Cubic, leaveTangent=-0.0405, tangentMode=RCTM_User, time=1, value=0.25}
         ]
        }
       }}
      Scale Curve=<5 3.5 2.5>
     )
     1=<0.8 0.8 0.8>
     2=?
     3=?
     Index=$Particles.VisibilityTag
     Num=::NewEnumerator0
    )
    ScaleA=true
    ScaleRGB=true
   )
   &ScaleMeshSize(Scale Factor=&VectorFromFloat(Value=&FloatFromCurve(
      FloatCurve=DataInterface() {propertyValues {Curve {
         defaultValue=0
         keys [
          {arriveTangent=4.1785, interpMode=RCIM_Cubic, leaveTangent=4.1785, tangentMode=RCTM_User, time=0, value=0}
          {arriveTangent=1.1179, interpMode=RCIM_Cubic, leaveTangent=1.1179, tangentMode=RCTM_User, time=0.2042, value=0.6228}
          {arriveTangent=0.0224, interpMode=RCIM_Cubic, leaveTangent=0.0224, tangentMode=RCTM_User, time=1, value=1}
         ]
        }}}
      Scale Curve=1.5
     ))) {module=ScaleMeshSize001}
  ]
  rendererClasses [&NiagaraMeshRendererProperties, &NiagaraSpriteRendererProperties]
  renderers [
   {rendererClass=&NiagaraMeshRendererProperties, rendererIndex=0}
   {rendererClass=&NiagaraSpriteRendererProperties, rendererIndex=1}
  ]
 }
 Intro {
  emitterSpawnScript [
   &InitializeAudioPlayer(Audio Player=DataInterface() {propertyValues {
      Attenuation=null
      Concurrency=null
      MaxPlaysPerTick=10
      SoundToPlay=/CosmeticCompanions/Assets/Audio_Shared/Interaction/Petting/SC_Mimosa_Interaction_React_Default
      bAllowLoopingOneShotSounds=false
      bLimitPlaysPerTick=true
      bOnlyActiveDuringGameplay=false
      bStopWhenComponentIsDestroyed=true
      ConfigurationUserParameter {parameter {name=None, typeDefHandle {registeredTypeIndex=1}, varData []}}
      ParameterNames []
     }})
   /Niagara/Modules/Audio/PlayAudio() {module=PlayAudio001}
  ]
  emitterUpdateScript [
   &EmitterState()
   &SpawnBurst_Instantaneous(Use Spawn Probability=false, UseLoopCountLimit=false)
   &SpawnBurst_Instantaneous(Spawn Group=1, Use Spawn Probability=false, UseLoopCountLimit=false) {module=SpawnBurst_Instantaneous001}
   /SetVariables(
    Emitter.AnimCurve=DataInterface() {propertyValues {Curve {
       defaultValue=0.5
       keys [
        {tangentMode=RCTM_User, time=0.05, value=-0.2}
        <at 0.149 -0.5>
        <at 0.15 0.38>
        <at 0.5 0.38>
        <at 0.5012 -0.2>
        <at 0.5395 0>
        <at 0.79 0.0097>
        <at 0.792 0.2395>
       ]
      }}}
    Emitter.UV_Distortion_Anim:NiagaraFloat=&FloatFromCurve(
     FloatCurve=DataInterface() {propertyValues {Curve {keys [
         {interpMode=RCIM_Cubic, leaveTangent=0.1172, tangentMode=RCTM_Break, time=0.2, value=0}
         {arriveTangent=-4.5741, interpMode=RCIM_Cubic, leaveTangent=-4.5741, tangentMode=RCTM_User, time=0.3036, value=1}
         <at 0.45 0>
         {interpMode=RCIM_Cubic, leaveTangent=-0.315, tangentMode=RCTM_Break, time=0.57, value=0}
         {arriveTangent=-3.6522, interpMode=RCIM_Cubic, leaveTangent=-3.6522, tangentMode=RCTM_User, time=0.7039, value=1}
         <at 0.8 0>
         <at 0.9413 0.0029>
         <at 0.9699 1>
        ]}}}
     CurveIndex=$Emitter.Age
    )
    Emitter.UV_Distortion_U_Offset:NiagaraFloat=&FloatFromCurve(FloatCurve=$Emitter.AnimCurve, CurveIndex=$Emitter.Age)
    Emitter.UV_Distortion_U_Scale:NiagaraFloat=&Lerp_Float(
     A=0.7
     B=0.5
     Alpha=&FloatFromCurve(
      FloatCurve=DataInterface() {propertyValues {Curve {keys [
          {interpMode=RCIM_Cubic, leaveTangent=0.1172, tangentMode=RCTM_Break, time=0.2, value=0}
          {arriveTangent=-4.5741, interpMode=RCIM_Cubic, leaveTangent=-4.5741, tangentMode=RCTM_User, time=0.3036, value=1}
          <at 0.45 0>
          {interpMode=RCIM_Cubic, leaveTangent=-0.315, tangentMode=RCTM_Break, time=0.57, value=0}
          {arriveTangent=-3.6522, interpMode=RCIM_Cubic, leaveTangent=-3.6522, tangentMode=RCTM_User, time=0.7039, value=1}
          <at 0.8 0>
          <at 0.9413 0.0029>
          <at 0.9699 1>
         ]}}}
      CurveIndex=$Emitter.Age
     )
    )
    Emitter.UV_Distortion_V_Offset=0.2
    Emitter.UV_Distortion_V_Scale:NiagaraFloat=&Lerp_Float(
     A=0.4
     B=0.2
     Alpha=&FloatFromCurve(
      FloatCurve=DataInterface() {propertyValues {Curve {keys [
          {interpMode=RCIM_Cubic, leaveTangent=0.1172, tangentMode=RCTM_Break, time=0.15, value=0}
          {arriveTangent=-4.5741, interpMode=RCIM_Cubic, leaveTangent=-4.5741, tangentMode=RCTM_User, time=0.2536, value=1}
          <at 0.4 0>
          {interpMode=RCIM_Cubic, leaveTangent=-0.315, tangentMode=RCTM_Break, time=0.52, value=0}
          {arriveTangent=-3.6522, interpMode=RCIM_Cubic, leaveTangent=-3.6522, tangentMode=RCTM_User, time=0.6539, value=1}
          <at 0.75 0>
          <at 0.8913 0.0029>
          <at 0.9199 1>
         ]}}}
      CurveIndex=$Emitter.Age
     )
    )
   ) {isSetParameters=true}
  ]
  particleSpawnScript [
   /SetVariables(Particles.VisibilityTag=$Emitter.SpawnGroup) {isSetParameters=true}
   &InitializeParticle(
    Color=<rgba 0.402 0.1441 0.0296 1>
    Lifetime=$System.IntroLength
    Ribbon Width=10
    Ribbon Width Mode=::NewEnumerator1
    Sprite Rotation Mode=::NewEnumerator0
    Sprite Size Mode=::NewEnumerator3
    Uniform Sprite Size=&Multiply_Float(A=25, B=$System.GlobalScale)
    UsePositionOffset=false
    Write Material Random=true
   )
   /SetVariables() {isSetParameters=true}
  ]
  particleUpdateScript [
   &ParticleState()
   &ScaleSpriteSize(
    Non-Uniform Curve Scale=&MergeHelper_Vector2Switch(
     0=<1.1 1.1>
     1=&Multiply_Vector2D_ByFloat(
      Float=&FloatFromCurve(FloatCurve=DataInterface() {propertyValues {Curve {
          defaultValue=1
          keys [
           {tangentMode=RCTM_User, time=0, value=1}
           {arriveTangent=-0.0405, interpMode=RCIM_Cubic, leaveTangent=-0.0405, tangentMode=RCTM_User, time=0.7009, value=1}
           {arriveTangent=-0.3622, interpMode=RCIM_Cubic, leaveTangent=-0.3622, tangentMode=RCTM_User, time=1, value=0.95}
          ]
         }}})
      Vector2D=<1.2 1.2>
     )
     2=?
     3=?
     4=?
     Index=$Particles.VisibilityTag
     Num=::NewEnumerator0
    )
    Non-Uniform Curve Sprite Scale=DataInterface() {propertyValues {
      XCurve {
       defaultValue=0
       keys [
        {arriveTangent=6.6919, interpMode=RCIM_Cubic, leaveTangent=6.6919, tangentMode=RCTM_User, time=0, value=0.3}
        {arriveTangent=-0.103, interpMode=RCIM_Cubic, leaveTangent=-0.103, tangentMode=RCTM_User, time=0.2979, value=1.0933}
        {arriveTangent=-0.5865, interpMode=RCIM_Cubic, leaveTangent=-0.5865, time=0.5504, value=0.5141}
        {arriveTangent=4.6226, interpMode=RCIM_Cubic, leaveTangent=4.6226, tangentMode=RCTM_User, time=0.6477, value=0.8881}
        {arriveTangent=-1.1057, interpMode=RCIM_Cubic, leaveTangent=-1.1057, time=0.743, value=1.051}
        {arriveTangent=-8.7807, interpMode=RCIM_Cubic, leaveTangent=-8.7807, tangentMode=RCTM_User, time=0.8867, value=0.6238}
        {arriveTangent=6.566, arriveTangentWeight=0.1667, interpMode=RCIM_Cubic, leaveTangent=6.566, tangentMode=RCTM_User, time=0.95, value=1.3}
        {
         arriveTangent=35.933
         arriveTangentWeight=0.0333
         interpMode=RCIM_Cubic
         leaveTangent=35.933
         tangentMode=RCTM_User
         time=1.0007
         value=2.9135
        }
       ]
      }
      YCurve {
       defaultValue=0
       keys [
        <at 0 0.3 RCIM_Cubic>
        {arriveTangent=0.131, interpMode=RCIM_Cubic, leaveTangent=0.131, tangentMode=RCTM_User, time=0.3855, value=1.2064}
        {arriveTangent=-5.7157, interpMode=RCIM_Cubic, leaveTangent=-5.7157, tangentMode=RCTM_User, time=0.6782, value=0.8601}
        {arriveTangent=-5.2362, interpMode=RCIM_Cubic, leaveTangent=-0.7602, tangentMode=RCTM_Break, time=0.7424, value=0.4574}
        {arriveTangent=0.749, interpMode=RCIM_Cubic, leaveTangent=0.749, tangentMode=RCTM_User, time=0.8792, value=0.3801}
        {arriveTangent=6.566, arriveTangentWeight=0.1667, interpMode=RCIM_Cubic, leaveTangent=6.566, tangentMode=RCTM_User, time=0.95, value=1.3}
        {
         arriveTangent=35.933
         arriveTangentWeight=0.0333
         interpMode=RCIM_Cubic
         leaveTangent=35.933
         tangentMode=RCTM_User
         time=0.9662
         value=3.1693
        }
       ]
      }
     }}
    Scale Sprite Size Mode=::NewEnumerator3
   )
   &ScaleColor(
    Scale Alpha=&FloatFromCurve(FloatCurve=DataInterface() {propertyValues {Curve {
        defaultValue=0
        keys [
         {arriveTangent=9.0725, interpMode=RCIM_Cubic, leaveTangent=9.0725, tangentMode=RCTM_User, time=0, value=0}
         {arriveTangent=-0.2441, leaveTangent=-0.2441, time=0.1408, value=0.9847}
         {arriveTangent=-1.146, leaveTangent=-1.146, time=1, value=1}
         {arriveTangent=-0.0405, interpMode=RCIM_Cubic, leaveTangent=-0.0405, tangentMode=RCTM_User, time=1, value=0}
        ]
       }}})
    ScaleA=true
    ScaleRGB=true
   )
   &OffsetPosition(Position Offset=&Lerp_Vector(
     A=$System.HeartIntroSpawnPosition
     B=$System.BurstPosition
     Alpha=&FloatFromCurve(FloatCurve=DataInterface() {propertyValues {Curve {keys [
          {arriveTangent=3.0606, interpMode=RCIM_Cubic, leaveTangent=3.0606, tangentMode=RCTM_User, time=0, value=0}
          {arriveTangent=-0.1073, tangentMode=RCTM_User, time=1, value=1}
         ]}}})
    ))
   &DynamicMaterialParameters(
    Index 0 Param 1=1
    Index 0 Param 2=1
    Module.Index0Float0Write=true
    Module.Index0Float1Write=true
    Module.Index0Float2Write=true
    Module.Index0Float3Write=true
   )
   /SetVariables(Particles.DMP_Lion_00:Vector4f=<0 1 0 0>) {isSetParameters=true}
   /SetVariables(Particles.Glow Color=<rgba 0 0 0 1>) {isSetParameters=true}
  ]
  rendererClasses [&NiagaraSpriteRendererProperties]
  renderers [
   {rendererClass=&NiagaraSpriteRendererProperties}
   {rendererClass=&NiagaraSpriteRendererProperties, rendererIndex=1}
  ]
 }
 Spikes {
  simTarget=GPUComputeSim
  emitterUpdateScript [
   &EmitterState(Life Cycle Mode=::NewEnumerator1, Loop Behavior=::NewEnumerator1, Loop Delay=0.03, Loop Duration Mode=::NewEnumerator1, UseLoopDelay=false)
   &SpawnBurst_Instantaneous(
    Spawn Count=&RandomRangeInt(Minimum=3, Maximum=4)
    Spawn Time=&Add_Float(A=$System.IntroLength, B=0)
    Use Spawn Probability=false
    UseLoopCountLimit=false
   )
  ]
  particleSpawnScript [
   &InitializeParticle(
    Color=<rgba 1 0.5333 0 1>
    Lifetime Max=0.4
    Lifetime Min=0.2
    Lifetime Mode=::NewEnumerator1
    Sprite Size Max=<70 150>
    Sprite Size Min=<50 40>
    Sprite Size Mode=::NewEnumerator2
    UsePositionOffset=false
    Write Material Random=true
   )
   &SubUVAnimation(End Frame=0, SubUV Animation Mode:ENiagara_SubUVLookupMode=::NewEnumerator2)
   &ShapeLocation(
    Handle Radius=0
    Large Radius=&RandomRangeFloat(Minimum=30, Maximum=25)
    MaskSpawnGroup=false
    Non Uniform Scale=&VectorFromFloat(Value=$System.GlobalScale)
    Offset=$System.BurstPosition
    Offset Mode=::NewEnumerator0
    Rotation Angle:NiagaraFloat=&RandomRangeFloat(Minimum=0, Maximum=45)
    Rotation Angle Type:ENiagara_AngleInput=::NewEnumerator0
    Rotation Axis=&ReturnCameraVector(Camera Vectors=::NewEnumerator1, Coordinate Space=::Local)
    Rotation Mode=::NewEnumerator1
    Shape Primitive=::NewEnumerator3
    Torus Distribution Mode:ENiagaraTorusDistributionMode=::NewEnumerator1
    U Position=&ReturnNormalizedExecIndex(Normalized Index Includes One=false)
    UseOwnerScale=true
    V Position=0
   )
   &AddVelocity(
    Velocity=$Particles.ShapeLocation.ShapeNormal
    Velocity Speed Scale=&RandomRangeFloat(Minimum=1, Maximum=50)
   )
  ]
  particleUpdateScript [
   &ParticleState()
   &DynamicMaterialParameters(
    Index 0 Param 1=3
    Index 0 Param 2=10
    Index 0 Param 4=&FloatFromCurve(FloatCurve=DataInterface() {propertyValues {Curve {keys [<at 0.4 0>,<at 1 1>]}}})
    Module.Index0Float0Write=true
    Module.Index0Float1Write=true
    Module.Index0Float2Write=true
    Module.Index0Float3Write=true
   )
   &ScaleSpriteSize(
    Uniform Curve Scale=$System.GlobalScale
    Uniform Curve Sprite Scale=DataInterface() {propertyValues {Curve {keys [{tangentMode=RCTM_User, time=0, value=1}, {tangentMode=RCTM_User, time=1, value=1.5}]}}}
   ) {module=ScaleSpriteSize001}
   &Drag(
    Drag=&FloatFromCurve(FloatCurve=DataInterface() {propertyValues {Curve {keys [<at 0 6>,<at 0.2 8>]}}}, CurveIndex=$Emitter.Age, Scale Curve=0.5)
    Use Linear Drag=true
    Use Rotational Drag=false
   )
   &SolveForcesAndVelocity(Clamp Velocity=false, Limit Acceleration=false)
   &ScaleSpriteSize(
    Non-Uniform Curve Scale=<0.5 0.3>
    Non-Uniform Curve Sprite Scale=DataInterface() {propertyValues {
      XCurve {keys [{tangentMode=RCTM_User, time=0, value=0.7}, {tangentMode=RCTM_User, time=1, value=1}]}
      YCurve {keys [{tangentMode=RCTM_User, time=0, value=1}, {tangentMode=RCTM_User, time=1, value=0.6}]}
     }}
    Scale Sprite Size Mode=::NewEnumerator3
   )
   &SubUVAnimation(UseEndFrame=false, UseStartFrame=false) {module=SubUVAnimation001}
   /SetVariables(Particles.SpriteSizeOutline:Vector2f=&Multiply_Vector2D(A=$Particles.SpriteSize, B=<0.925 0.95>)) {isSetParameters=true}
   &FN_Sprite_Attributes(
    Color=&MakeLinearColorFromVectorAndFloat("Vector (RGB)"=&MakeVectorFromLinearColorRGB(Color=<rgba 0 0 0 1>), "Float (Alpha)"=5)
    Create Color Attribute=true
    Create SpriteSize Attribute=true
    SpriteSize=<120 220>
    Write to Module Namespace=true
   )
  ]
  rendererClasses [&NiagaraSpriteRendererProperties]
  renderers [
   {rendererClass=&NiagaraSpriteRendererProperties}
   {rendererClass=&NiagaraSpriteRendererProperties, rendererIndex=1}
  ]
 }
}
summary {
 systemName=NS_ReactFX_AboutJury
 emitters [
  {emitterName=Intro, rendererClasses [&NiagaraSpriteRendererProperties]}
  {emitterName=BubblesHearts, simTarget=GPUComputeSim, rendererClasses [&NiagaraSpriteRendererProperties]}
  {emitterName=CrackGraphic, simTarget=GPUComputeSim, rendererClasses [&NiagaraSpriteRendererProperties]}
  {emitterName=Spikes, simTarget=GPUComputeSim, rendererClasses [&NiagaraSpriteRendererProperties]}
  {emitterName=Glow, rendererClasses [&NiagaraMeshRendererProperties, &NiagaraSpriteRendererProperties]}
 ]
 userVariables [
  User.bIsFrontEnd {type=NiagaraBool}
  User.bIsFrontEndPreview {type=NiagaraBool}
  User.FrontEndScale {type=NiagaraFloat, defaultValue=0.6, description=""}
 ]
}
