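# Cosmetic companion "abrupt sudden" reaction effect that plays a petting-reaction sound with a short-lived icon and glitch burst.
# Heart bursts four sprites, using spawn groups for per-layer lifetime, scale, offset and jitter, plus inner, ray and glow parameters.
# Glitches emits brief, randomly stretched meshes in a delayed window, with a washed-out rainbow flicker.
# Front-end modifiers and time-of-day values adjust placement, size, spawn rate and brightness.
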NiagaraSystem /CosmeticCompanions/FX/React/NS_ReactFX_AbruptSudden
dependencies {
 usedDataInterfaces [
  &NiagaraDataInterfaceAudioPlayer
  &NiagaraDataInterfaceCurve
  &NiagaraDataInterfaceVectorCurve
  &NiagaraDataInterfaceVector2DCurve
  &NiagaraDataInterfaceColorCurve
 ]
 usedDynamicInputs []
 usedModules [
  &self:SystemSpawnScript.SystemScriptSource.SystemScriptGraph.NiagaraNodeAssignment_0.SetVariables_#1
  &self:SystemSpawnScript.SystemScriptSource.SystemScriptGraph.NiagaraNodeAssignment_1.SetVariables_#2
  &self:SystemSpawnScript.SystemScriptSource.SystemScriptGraph.NiagaraNodeAssignment_2.SetVariables_#3
  &self:SystemSpawnScript.SystemScriptSource.SystemScriptGraph.NiagaraNodeAssignment_3.SetVariables_#4
  &SystemState
  &self:SystemSpawnScript.SystemScriptSource.SystemScriptGraph.NiagaraNodeAssignment_4.SetVariables_#5
 ]
}
emitters {
 Glitches {
  emitterUpdateScript [
   &EmitterState(Life Cycle Mode=::NewEnumerator1, Loop Behavior=::NewEnumerator1, Loop Delay=0.7, Loop Duration=0.4, UseLoopDelay=true)
   &SpawnRate(SpawnRate=&FN_FrontEndFloatModifier(Gameplay=130, Front End=100), Use Spawn Probability=false)
  ]
  particleSpawnScript [
   &InitializeParticle(
    Color=&MultiplyLinearColor(Linear Color=<rgba 1 0.8026 0.2645 1>, Scale Factor=&MakeVector4_FromLinearColor(LinearColor=$System.TOD))
    Lifetime Max=0.2
    Lifetime Min=0.1
    Lifetime Mode=::NewEnumerator1
    Mesh Scale Max=<3 50 10>
    Mesh Scale Min=<0.5 0.5 5>
    Mesh Scale Mode=::NewEnumerator2
    Position Offset=&Add_Vector(A=$System.BurstPosition, B=&FN_FrontEndVector3Modifier(Gameplay=<0 0 -83>, Front End=<0 0 -63>))
    UsePositionOffset=true
    Write Material Random=true
   )
   &ShapeLocation(
    Cone Angle=33
    Cone Length:NiagaraFloat=&FN_FrontEndFloatModifier(Gameplay=75, Front End=33)
    MaskSpawnGroup=false
    Override Seed=false
    Shape Primitive=::NewEnumerator5
    UseOwnerScale=true
   )
   &CameraOffset(Camera Offset Amount=5)
  ]
  particleUpdateScript [
   &ParticleState()
   &ScaleMeshSize(Scale Factor=&Multiply_VectorByFloat(
     Vector=<1 1 1>
     Float=&Multiply_Float(
      A=0.07
      B=&FloatFromCurve(
       FloatCurve=DataInterface() {propertyValues {Curve {keys [
           {arriveTangent=4.1785, time=0, value=0}
           <at 0.5 1.1>
           {arriveTangent=-0.0405, interpMode=RCIM_Cubic, leaveTangent=-0.0405, tangentMode=RCTM_User, time=1, value=1.5}
          ]}}}
       Scale Curve=0.176
      )
     )
    ))
   &Drag(Drag=4, Use Linear Drag=true, Use Rotational Drag=true)
   &SolveForcesAndVelocity(Clamp Velocity=false, Limit Acceleration=false)
   &ScaleColor(
    Scale Alpha=&Multiply_Float(
     A=&FloatFromCurve(
      FloatCurve=DataInterface() {propertyValues {Curve {keys [<at 0 1>,<at 1 0>]}}}
      Scale Curve=&RandomRangeFloat(Minimum=0.4, Maximum=1)
     )
     B=&Multiply_Float(
      A=&RandomRangeFloat(Minimum=0.2, Maximum=0.7)
      B=&FN_FrontEndFloatModifier(Gameplay=1, Front End=0.6)
     )
    )
    Scale RGB=&Multiply_VectorByFloat(
     Vector=&MakeVectorFromLinearColorRGB(Color=&LerpLinearColors(
       StartColor=&ColorFromCurve(ColorCurve=DataInterface() {propertyValues {
          AlphaCurve {keys []}
          BlueCurve {keys [
            <at 0 5 RCIM_Cubic>
            <at 0.1326 5>
            <at 0.2869 0>
            <at 0.4134 0>
            <at 0.5108 5>
            <at 0.6236 0.0803>
            {arriveTangent=26.4359, interpMode=RCIM_Cubic, leaveTangent=26.4359, time=0.7107, value=0}
            <at 0.8097 5>
           ]}
          GreenCurve {keys [
            <at 0 4.702 RCIM_Cubic>
            <at 0.1326 0>
            <at 0.2869 2.0314>
            <at 0.4134 0.0442>
            <at 0.5108 0.7585>
            <at 0.6236 5>
            {arriveTangent=-21.532, interpMode=RCIM_Cubic, leaveTangent=-21.532, time=0.7107, value=0.3634}
            <at 0.8097 0.9929>
           ]}
          RedCurve {keys [<at 0 0 RCIM_Cubic>,<at 0.1326 2.465>,<at 0.2869 5>,<at 0.4134 5>,<at 0.5108 4.6237>,<at 0.6236 0>,<at 0.7107 1 RCIM_Cubic>,<at 0.8097 0>]}
         }})
       EndColor=<rgba 1 1 1 1>
       LerpFactor=0.8542
      ))
     Float=&FloatFromCurve(
      FloatCurve=DataInterface() {propertyValues {Curve {keys [{arriveTangent=-2.0684, interpMode=RCIM_Cubic, leaveTangent=-2.0684, tangentMode=RCTM_User, time=0, value=2}, <at 1 0>]}}}
      Scale Curve=&Multiply_Float(A=&Multiply_Float(A=3, B=1), B=1)
     )
    )
    ScaleA=true
    ScaleRGB=true
   )
  ]
  rendererClasses [&NiagaraMeshRendererProperties]
  renderers [{rendererClass=&NiagaraMeshRendererProperties, rendererIndex=0}]
 }
 Heart {
  emitterSpawnScript [
   &InitializeAudioPlayer(Audio Player=DataInterface() {propertyValues {
      Attenuation=null
      Concurrency=null
      MaxPlaysPerTick=10
      SoundToPlay=/CosmeticCompanions/Assets/Audio_Shared/Interaction/Petting/SC_Mimosa_Interaction_React_Default
      bAllowLoopingOneShotSounds=false
      bLimitPlaysPerTick=true
      bOnlyActiveDuringGameplay=false
      bStopWhenComponentIsDestroyed=true
      ConfigurationUserParameter {parameter {name=None, typeDefHandle {registeredTypeIndex=1}, varData []}}
      ParameterNames []
     }})
   /Niagara/Modules/Audio/PlayAudio() {module=PlayAudio001}
  ]
  emitterUpdateScript [
   &EmitterState()
   &SpawnBurst_Instantaneous(Spawn Count=4, Use Spawn Probability=false, UseLoopCountLimit=false) {module=SpawnBurst_Instantaneous001}
  ]
  particleSpawnScript [
   &FN_SetSpawnGroup()
   &InitializeParticle(
    Lifetime=&MergeHelper_FloatSwitch(
     0=&Add_Float(A=$System.IntroLength, B=0.4)
     1=&Add_Float(A=$System.IntroLength, B=0.55)
     2=&Add_Float(A=$System.IntroLength, B=0.43)
     3=&Add_Float(A=$System.IntroLength, B=0.4)
     Index=&SpawnGroup()
     Num=::NewEnumerator2
    )
    Sprite Size Mode=::NewEnumerator3
    Uniform Sprite Size=&Multiply_Float(A=20, B=$System.FrontendScale)
    UsePositionOffset=false
    Write Material Random=true
   )
   &CameraOffset(Camera Offset Amount=&MergeHelper_FloatSwitch(0=5, 1=5, 2=3, 3=1, Index=&SpawnGroup(), Num=::NewEnumerator2))
   /SetVariables(Particles.VisibilityTag=&SpawnGroup()) {isSetParameters=true}
   &SpriteFacingAndAlignment(Sprite Alignment=<0 0 1>, WriteAlignment=true, WriteFacing=true)
  ]
  particleUpdateScript [
   &ParticleState()
   &ScaleSpriteSize(
    Scale Factor=&MergeHelper_Vector2Switch(0=<1.7 1.7>, 1=<1 1>, 2=<1.9529 2.3456>, 3=?, 4=?, Index=&SpawnGroup(), Num=::NewEnumerator1)
    Scale Sprite Size Mode=::NewEnumerator2
   ) {module=ScaleSpriteSize001}
   &ScaleColor(
    Scale Alpha=&FloatFromCurve(
     FloatCurve=DataInterface() {propertyValues {Curve {
        defaultValue=0
        keys [
         {arriveTangent=-0.0743, interpMode=RCIM_Constant, time=0, value=0}
         {arriveTangent=3.2463, interpMode=RCIM_Cubic, leaveTangent=3.2463, time=0.2537, value=0}
         {arriveTangent=5.3486, interpMode=RCIM_Cubic, leaveTangent=5.3486, time=0.308, value=1}
         <at 0.3661 0.6013 RCIM_Cubic>
         {interpMode=RCIM_Cubic, tangentMode=RCTM_Break, time=0.418, value=1}
         {interpMode=RCIM_Cubic, tangentMode=RCTM_User, time=0.7433, value=1}
         {arriveTangent=-8.414, tangentMode=RCTM_User, time=0.9518, value=0}
        ]
       }}}
     Scale Curve=&Multiply_Float(A=0.9, B=1)
    )
    Scale RGB=&MakeVectorFromLinearColorRGB(Color=&Multiply_LinearColorByFloat(
      Linear Color=<rgba 1 1 1 1>
      Float=&FN_FrontEndFloatModifier(
       Gameplay=&FNTimeOfDayScalar(Night Scale=0.66, Morning Scale=0.9, Evening Scale=0.85)
       Front End=0.6
      )
     ))
    ScaleA=true
    ScaleRGB=true
   )
   &DynamicMaterialParameters(
    Index 0 Param 1=1
    Index 0 Param 2=0.9
    Index 0 Param 3=&Multiply_Float(
     A=1.25
     B=&FN_FrontEndFloatModifier(Gameplay=&FNTimeOfDayScalar(Night Scale=1, Morning Scale=1, Evening Scale=1), Front End=0.5)
    )
    Index 0 Param 4=1
    Module.Index0Float0Write=true
    Module.Index0Float1Write=true
    Module.Index0Float2Write=true
    Module.Index0Float3Write=true
   )
   /SetVariables(Particles.Position=&AddVectorToPosition(
     Position=Linked() {linkedVariable {name=System.BurstPosition, type=Vector3f}}
     Vector=&MergeHelper_VectorSwitch(
      0=&FN_FrontEndVector3Modifier(Gameplay=<0 0 -20>, Front End=<0 0 -37>)
      1=&Add_Vector(
       A=<0 0 -65>
       B=&VectorFromCurve(VectorCurve=DataInterface() {propertyValues {XCurve {keys [<at 0 0>,<at 1 0>]}, YCurve {keys [<at 0 0>,<at 1 0>]}, ZCurve {keys [<at 0 -12>,<at 0.07 0>,<at 0.9233 0>,<at 1 -11>]}}})
      )
      2=&FN_FrontEndVector3Modifier(Gameplay=<0 0 -58>, Front End=<0 0 -61>)
      3=?
      Index=&SpawnGroup()
      Num=::NewEnumerator1
     )
    )) {isSetParameters=true}
   &JitterPosition(Jitter Amount:NiagaraFloat=&FloatFromCurve(
     FloatCurve=DataInterface() {propertyValues {Curve {
        defaultValue=1
        keys [
         {interpMode=RCIM_Cubic, tangentMode=RCTM_User, time=0.551, value=0}
         {arriveTangent=2.2269, interpMode=RCIM_Cubic, leaveTangent=2.2269, time=0.8036, value=1}
         <at 1 1 RCIM_Cubic>
        ]
       }}}
     Scale Curve=&MergeHelper_FloatSwitch(0=2, 1=0, 2=0, 3=?, Index=&SpawnGroup(), Num=::NewEnumerator1)
    ))
   /SetVariables(
    Particles.Inner Color:LinearColor=&Multiply_LinearColorByFloat(Linear Color=$System.TOD, Float=0.6)
    Particles.Inner Size:Vector2f=&Multiply_Vector2D_ByFloat(
     Float=&FloatFromCurve(
      FloatCurve=DataInterface() {propertyValues {Curve {
         defaultValue=0
         keys [
          {arriveTangent=10.6939, interpMode=RCIM_Cubic, leaveTangent=10.6939, tangentMode=RCTM_User, time=0, value=0.3}
          {arriveTangent=2.2222, time=0.0476, value=2}
          {arriveTangent=-2.1001, leaveTangent=-2.1001, time=0.95, value=2}
          <at 1 0.3 RCIM_Cubic>
         ]
        }}}
      Scale Curve=0.7
     )
     Vector2D=$Particles.Initial.SpriteSize
    )
   ) {isSetParameters=true}
   /SetVariables(
    Particles.RayColor:LinearColor=&MakeLinearColorFromVectorAndFloat(
     "Vector (RGB)"=&MakeVectorFromLinearColorRGB(Color=&MultiplyLinearColor(Linear Color=<rgba 3 2.4046 0.9375 1>, Scale Factor=&MakeVector4_FromLinearColor(LinearColor=$System.TOD)))
     "Float (Alpha)"=&FloatFromCurve(
      FloatCurve=DataInterface() {propertyValues {Curve {keys [<at 0.0696 0 RCIM_Cubic>, <at 0.3 1>, {arriveTangent=-1.4286, interpMode=RCIM_Cubic, leaveTangent=-1.4286, time=0.9, value=1}, <at 1 0>]}}}
      Scale Curve=&Multiply_Float(A=0.2, B=&FN_FrontEndFloatModifier(Gameplay=1, Front End=0.6))
     )
    )
    Particles.RaySize:Vector2f=&Vector2DFromCurve(
     Vector2Curve=DataInterface() {propertyValues {
       XCurve {keys [
         <at 0.0439 0 RCIM_Cubic>
         {arriveTangent=0.0391, interpMode=RCIM_Cubic, leaveTangent=0.0391, tangentMode=RCTM_User, time=0.2106, value=1}
         {arriveTangent=-0.0366, interpMode=RCIM_Cubic, leaveTangent=-0.0366, tangentMode=RCTM_User, time=0.9153, value=1}
         <at 1 0 RCIM_Cubic>
        ]}
       YCurve {keys [<at 0 1 RCIM_Cubic>,<at 0.22 1>]}
      }}
     ScaleCurve=&Multiply_Vector2D(A=$Particles.SpriteSize, B=<0.9 0.75>)
    )
   ) {isSetParameters=true}
   /SetVariables(
    Particles.GlowColor=&MultiplyLinearColor(
     Linear Color=$Particles.Color
     Scale Factor=&MakeVector4_FromLinearColor(LinearColor=&Multiply_LinearColorByFloat(
       Linear Color=<rgba 2 1.2696 0.5729 0.061>
       Float=&FN_FrontEndFloatModifier(Gameplay=1, Front End=0.7)
      ))
    )
    Particles.GlowSize=&Multiply_Vector2D_ByFloat(Float=&Multiply_Float(A=50, B=$System.FrontendScale), Vector2D=<1 1>)
   ) {isSetParameters=true}
  ]
  rendererClasses [&NiagaraSpriteRendererProperties]
  renderers [
   {rendererClass=&NiagaraSpriteRendererProperties}
   {rendererClass=&NiagaraSpriteRendererProperties, rendererIndex=1}
   {rendererClass=&NiagaraSpriteRendererProperties, rendererIndex=2}
   {rendererClass=&NiagaraSpriteRendererProperties, rendererIndex=3}
  ]
 }
}
summary {
 systemName=NS_ReactFX_AbruptSudden
 emitters [
  {emitterName=Heart, rendererClasses [&NiagaraSpriteRendererProperties]}
  {emitterName=Glitches, rendererClasses [&NiagaraMeshRendererProperties]}
 ]
 userVariables [
  User.bIsFrontEnd {type=NiagaraBool}
  User.bIsFrontEndPreview {type=NiagaraBool}
 ]
}
