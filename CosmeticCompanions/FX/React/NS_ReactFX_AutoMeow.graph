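# Niagara reaction effect for the AutoMeow cosmetic companion, combining a heart, cyan light rays and color-shifting glitch meshes with the shared petting sound.
# Heart bursts seven sprites, using spawn groups to select renderer, material parameters and inner/glow curves.
# Glitches and flickering LightRays spawn below a shared BurstPosition and fade by system age.
# Front-end modifiers and System.FrontendScale adjust size and brightness, while time-of-day values tint or dim colors.
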NiagaraSystem /CosmeticCompanions/FX/React/NS_ReactFX_AutoMeow
dependencies {
 usedDataInterfaces [
  &NiagaraDataInterfaceAudioPlayer
  &NiagaraDataInterfaceCurve
  &NiagaraDataInterfaceColorCurve
  &NiagaraDataInterfaceSpriteRendererInfo
  &NiagaraDataInterfaceVector2DCurve
 ]
 usedDynamicInputs [&MakeFloatFromVector2d, &VectorFromFloat]
 usedModules [
  &self:SystemSpawnScript.SystemScriptSource.SystemScriptGraph.NiagaraNodeAssignment_0.SetVariables_#1
  &self:SystemSpawnScript.SystemScriptSource.SystemScriptGraph.NiagaraNodeAssignment_1.SetVariables_#2
  &self:SystemSpawnScript.SystemScriptSource.SystemScriptGraph.NiagaraNodeAssignment_2.SetVariables_#3
  &self:SystemSpawnScript.SystemScriptSource.SystemScriptGraph.NiagaraNodeAssignment_3.SetVariables_#4
  &self:SystemSpawnScript.SystemScriptSource.SystemScriptGraph.NiagaraNodeAssignment_5.SetVariables_#5
  &SystemState
  &self:SystemSpawnScript.SystemScriptSource.SystemScriptGraph.NiagaraNodeAssignment_6.SetVariables_#6
 ]
}
emitters {
 Glitches {
  emitterUpdateScript [
   &EmitterState(Life Cycle Mode=::NewEnumerator1, Loop Behavior=::NewEnumerator1, Loop Delay=0.05, Loop Duration=1, UseLoopDelay=true)
   &SpawnRate(
    SpawnRate=&FloatFromCurve(
     FloatCurve=DataInterface() {propertyValues {Curve {keys [<at 0.1 1>,<at 0.6 0>]}}}
     CurveIndex=$System.Age
     Scale Curve=&FN_FrontEndFloatModifier(Gameplay=130, Front End=100)
    )
    Use Spawn Probability=false
   )
  ]
  particleSpawnScript [
   &InitializeParticle(
    Color=&MultiplyLinearColor(Linear Color=<rgba 0.143 0.8366 1 1>, Scale Factor=&MakeVector4_FromLinearColor(LinearColor=$System.TOD))
    Lifetime Max=0.4
    Lifetime Min=0.3
    Lifetime Mode=::NewEnumerator1
    Mesh Scale Max=<15 15 15>
    Mesh Scale Min=<0.5 0.5 5>
    Mesh Scale Mode=::NewEnumerator2
    Position Offset=&Add_Vector(A=$System.BurstPosition, B=&FN_FrontEndVector3Modifier(Gameplay=<0 0 -61>, Front End=<0 0 -61>))
    UsePositionOffset=true
    Write Material Random=true
   )
   &ShapeLocation(MaskSpawnGroup=false, Override Seed=false, Sphere Radius=1, UseOwnerScale=true)
   &CameraOffset(Camera Offset Amount=5)
   &AddVelocity(
    Velocity=&RandomRangeVector(Minimum=<-30 -30 100>, Maximum=<30 30 150>)
    Velocity Speed Scale=$System.FrontendScale
   )
  ]
  particleUpdateScript [
   &ParticleState()
   &ScaleMeshSize(Scale Factor=&Multiply_VectorByFloat(
     Vector=<1 1 1>
     Float=&Multiply_Float(
      A=0.07
      B=&FloatFromCurve(
       FloatCurve=DataInterface() {propertyValues {Curve {keys [
           {arriveTangent=4.1785, time=0, value=0}
           <at 0.5 1.1>
           {arriveTangent=-0.0405, interpMode=RCIM_Cubic, leaveTangent=-0.0405, tangentMode=RCTM_User, time=1, value=1.5}
          ]}}}
       Scale Curve=&Multiply_Float(A=0.085, B=$System.FrontendScale)
      )
     )
    ))
   &Drag(Drag=4, Use Linear Drag=true, Use Rotational Drag=true)
   &SolveForcesAndVelocity(Clamp Velocity=false, Limit Acceleration=false)
   &ScaleColor(
    Scale Alpha=&Multiply_Float(
     A=&FloatFromCurve(
      FloatCurve=DataInterface() {propertyValues {Curve {keys [<at 0 1>,<at 1 0>]}}}
      Scale Curve=&RandomRangeFloat(Minimum=1, Maximum=1)
     )
     B=&Multiply_Float(
      A=&RandomRangeFloat(Minimum=0.6, Maximum=1)
      B=&FN_FrontEndFloatModifier(Gameplay=1, Front End=0.6)
     )
    )
    Scale RGB=&Multiply_VectorByFloat(
     Vector=&MakeVectorFromLinearColorRGB(Color=&LerpLinearColors(
       StartColor=&ColorFromCurve(ColorCurve=DataInterface() {propertyValues {
          AlphaCurve {keys []}
          BlueCurve {keys [
            <at 0 5 RCIM_Cubic>
            <at 0.1326 5>
            <at 0.2869 0>
            <at 0.4134 0>
            <at 0.5108 5>
            <at 0.6236 0.0803>
            {arriveTangent=26.4359, interpMode=RCIM_Cubic, leaveTangent=26.4359, time=0.7107, value=0}
            <at 0.8097 5>
           ]}
          GreenCurve {keys [
            <at 0 4.702 RCIM_Cubic>
            <at 0.1326 0>
            <at 0.2869 2.0314>
            <at 0.4134 0.0442>
            <at 0.5108 0.7585>
            <at 0.6236 5>
            {arriveTangent=-21.532, interpMode=RCIM_Cubic, leaveTangent=-21.532, time=0.7107, value=0.3634}
            <at 0.8097 0.9929>
           ]}
          RedCurve {keys [<at 0 0 RCIM_Cubic>,<at 0.1326 2.465>,<at 0.2869 5>,<at 0.4134 5>,<at 0.5108 4.6237>,<at 0.6236 0>,<at 0.7107 1 RCIM_Cubic>,<at 0.8097 0>]}
         }})
       EndColor=<rgba 1 1 1 1>
       LerpFactor=0.8542
      ))
     Float=&FloatFromCurve(
      FloatCurve=DataInterface() {propertyValues {Curve {keys [{arriveTangent=-2.0684, interpMode=RCIM_Cubic, leaveTangent=-2.0684, tangentMode=RCTM_User, time=0, value=2}, <at 1 0>]}}}
      Scale Curve=&Multiply_Float(A=&Multiply_Float(A=3, B=1), B=1)
     )
    )
    ScaleA=true
    ScaleRGB=true
   )
  ]
  rendererClasses [&NiagaraMeshRendererProperties]
  renderers [{rendererClass=&NiagaraMeshRendererProperties, rendererIndex=0}]
 }
 Heart {
  emitterSpawnScript [
   &InitializeAudioPlayer(Audio Player=DataInterface() {propertyValues {
      Attenuation=null
      Concurrency=null
      MaxPlaysPerTick=10
      SoundToPlay=/CosmeticCompanions/Assets/Audio_Shared/Interaction/Petting/SC_Mimosa_Interaction_React_Default
      bAllowLoopingOneShotSounds=false
      bLimitPlaysPerTick=true
      bOnlyActiveDuringGameplay=false
      bStopWhenComponentIsDestroyed=true
      ConfigurationUserParameter {parameter {name=None, typeDefHandle {registeredTypeIndex=1}, varData []}}
      ParameterNames []
     }})
   /Niagara/Modules/Audio/PlayAudio() {module=PlayAudio001}
  ]
  emitterUpdateScript [
   &EmitterState()
   &SpawnBurst_Instantaneous(Spawn Count=7, Use Spawn Probability=false, UseLoopCountLimit=false) {module=SpawnBurst_Instantaneous001}
  ]
  particleSpawnScript [
   &FN_SetSpawnGroup(Number of Spawn Groups=7)
   &InitializeParticle(
    Lifetime=&Add_Float(A=$System.IntroLength, B=0.4)
    Sprite Size Mode=::NewEnumerator3
    Uniform Sprite Size=&Multiply_Float(A=20, B=$System.FrontendScale)
    UsePositionOffset=false
    Write Material Random=true
   )
   &CameraOffset(Camera Offset Amount=4)
   /SetVariables(Particles.VisibilityTag=&MergeHelper_IntSwitch(0=0, 1=0, 2=0, 3=1, 4=2, 5=2, 6=2, Index=&SpawnGroup(), Num=::NewEnumerator5)) {isSetParameters=true}
   /SetVariables(Particles.HeartScale=1.08) {isSetParameters=true}
  ]
  particleUpdateScript [
   &ParticleState()
   &ScaleSpriteSize(
    Scale Sprite Size Mode=::NewEnumerator0
    Uniform Scale Factor=&FloatFromCurve(FloatCurve=DataInterface() {propertyValues {Curve {keys [<at 0 0.6047 RCIM_Cubic>,<at 0.1897 0.8618>]}}}, Scale Curve=$Particles.HeartScale)
   ) {module=ScaleSpriteSize001}
   &ScaleColor(
    Scale Alpha=&FloatFromCurve(
     FloatCurve=DataInterface() {propertyValues {Curve {
        defaultValue=0
        keys [
         {arriveTangent=-0.0743, interpMode=RCIM_Cubic, time=0.08, value=0}
         {arriveTangent=0.0322, interpMode=RCIM_Cubic, leaveTangent=0.0322, tangentMode=RCTM_User, time=0.35, value=1}
         {interpMode=RCIM_Cubic, tangentMode=RCTM_User, time=0.4106, value=1}
         <at 0.8847 0 RCIM_Cubic>
        ]
       }}}
     Scale Curve=0.3
    )
    Scale RGB=&MakeVectorFromLinearColorRGB(Color=&Multiply_LinearColorByFloat(
      Linear Color=<rgba 1.3 1.3 1.3 1>
      Float=&FN_FrontEndFloatModifier(
       Gameplay=&FNTimeOfDayScalar(Night Scale=0.66, Morning Scale=0.9, Evening Scale=0.85)
       Front End=0.6
      )
     ))
    ScaleA=true
    ScaleRGB=true
   )
   &DynamicMaterialParameters(
    Index 0 Param 1=&MergeHelper_FloatSwitch(0=0, 1=0, 2=0, 3=1, 4=0, 5=0, 6=0, Index=&SpawnGroup(), Num=::NewEnumerator5)
    Index 0 Param 2=&MergeHelper_FloatSwitch(0=0, 1=0, 2=0, 3=0, 4=1, 5=1, 6=1, Index=&SpawnGroup(), Num=::NewEnumerator5)
    Index 0 Param 3=&Multiply_Float(
     A=1.25
     B=&FN_FrontEndFloatModifier(Gameplay=&FNTimeOfDayScalar(Night Scale=1, Morning Scale=1, Evening Scale=1), Front End=0.5)
    )
    Module.Index0Float0Write=true
    Module.Index0Float1Write=true
    Module.Index0Float2Write=true
    Module.Index0Float3Write=true
   )
   /SetVariables(Particles.Position=&AddVectorToPosition(
     Position=Linked() {linkedVariable {name=System.BurstPosition, type=Vector3f}}
     Vector=&FN_FrontEndVector3Modifier(Gameplay=<0 0 -37>, Front End=<0 0 -43>)
    )) {isSetParameters=true}
   /SetVariables(
    Particles.Inner Color:LinearColor=&Multiply_LinearColorByFloat(
     Linear Color=&MakeLinearColorFromVectorAndFloat(
      "Vector (RGB)"=&MakeVectorFromLinearColorRGB(Color=&MultiplyLinearColor(
        Linear Color=$System.TOD
        Scale Factor=&MakeVector4_FromFloat(Float=&FloatFromCurve(FloatCurve=DataInterface() {propertyValues {Curve {keys [
              <at 0.0994 0.8 RCIM_Cubic>
              {arriveTangent=0.6321, interpMode=RCIM_Cubic, leaveTangent=0.6321, time=0.2426, value=1.3}
              {arriveTangent=-0.019, interpMode=RCIM_Cubic, leaveTangent=-0.019, time=0.4158, value=1}
              <at 1 1.2856>
             ]}}}))
       ))
      "Float (Alpha)"=&MergeHelper_FloatSwitch(
       0=&FloatFromCurve(FloatCurve=DataInterface() {propertyValues {Curve {
           defaultValue=0
           keys [
            <at 0.08 0 RCIM_Cubic>
            {arriveTangent=2.2222, time=0.35, value=1}
            {arriveTangent=-2.1001, time=0.6, value=1}
            {arriveTangent=-0.2663, interpMode=RCIM_Cubic, leaveTangent=-0.2663, tangentMode=RCTM_User, time=0.8308, value=1}
            <at 0.95 0 RCIM_Cubic>
           ]
          }}})
       1=&FloatFromCurve(
        FloatCurve=DataInterface() {propertyValues {Curve {
           defaultValue=0
           keys [
            {arriveTangent=2.2222, time=0.0476, value=1}
            {arriveTangent=-1.0612, interpMode=RCIM_Cubic, leaveTangent=-1.0612, time=0.74, value=1}
            <at 0.99 0 RCIM_Cubic>
           ]
          }}}
        Scale Curve=0.16
       )
       2=&FloatFromCurve(
        FloatCurve=DataInterface() {propertyValues {Curve {
           defaultValue=0
           keys [
            {arriveTangent=2.2222, time=0.0476, value=1}
            {arriveTangent=-1.05, interpMode=RCIM_Cubic, leaveTangent=-1.05, time=0.82, value=1}
            <at 1 0 RCIM_Cubic>
           ]
          }}}
        Scale Curve=0.16
       )
       3=?
       Index=&SpawnGroup()
       Num=::NewEnumerator1
      )
     )
     Float=1
    )
    Particles.Inner Size:Vector2f=&Multiply_Vector2D_ByFloat(
     Float=&MergeHelper_FloatSwitch(
      0=&FloatFromCurve(
       FloatCurve=DataInterface() {propertyValues {Curve {
          defaultValue=0
          keys [
           {arriveTangent=2.2222, time=0.0476, value=2}
           {arriveTangent=-2.0502, interpMode=RCIM_Cubic, leaveTangent=-2.0502, time=0.7, value=2}
           <at 0.95 0.15 RCIM_Cubic>
          ]
         }}}
       Scale Curve=0.45
      )
      1=&FloatFromCurve(
       FloatCurve=DataInterface() {propertyValues {Curve {defaultValue=0, keys [{arriveTangent=2.2222, interpMode=RCIM_Constant, time=0.4468, value=0}, <at 0.7 2 RCIM_Constant>, <at 1 2 RCIM_Cubic>]}}}
       Scale Curve=0.45
      )
      2=&FloatFromCurve(
       FloatCurve=DataInterface() {propertyValues {Curve {
          defaultValue=0
          keys [{arriveTangent=2.2222, interpMode=RCIM_Constant, time=0.4468, value=0}, <at 0.75 2 RCIM_Constant>, <at 1 2 RCIM_Cubic>]
         }}}
       Scale Curve=0.225
      )
      3=?
      Index=&SpawnGroup()
      Num=::NewEnumerator1
     )
     Vector2D=&Multiply_Vector2D_ByFloat(Float=$Particles.HeartScale, Vector2D=$Particles.Initial.SpriteSize)
    )
   ) {isSetParameters=true}
   /SetVariables(
    Particles.GlowColor=&MultiplyLinearColor(
     Linear Color=$System.TOD
     Scale Factor=&MakeVector4_FromLinearColor(LinearColor=&Multiply_LinearColorByFloat(
       Linear Color=<rgba 0.8239 1.4147 1.6 0.038>
       Float=&Multiply_Float(
        A=&FN_FrontEndFloatModifier(Gameplay=1, Front End=0.7)
        B=&MergeHelper_FloatSwitch(
         0=0
         1=0
         2=0
         3=0
         4=&FloatFromCurve(FloatCurve=DataInterface() {propertyValues {Curve {
             defaultValue=0
             keys [<at 0.08 0 RCIM_Cubic>, {arriveTangent=2.2222, time=0.35, value=1}, {arriveTangent=-2.1001, time=0.6, value=1}, <at 1 1 RCIM_Cubic>]
            }}})
         5=&FloatFromCurve(
          FloatCurve=DataInterface() {propertyValues {Curve {
             defaultValue=0
             keys [
              {arriveTangent=2.2222, time=0.0476, value=1}
              {arriveTangent=-1.0612, interpMode=RCIM_Cubic, leaveTangent=-1.0612, time=0.74, value=1}
              <at 0.99 0 RCIM_Cubic>
             ]
            }}}
          Scale Curve=0.7
         )
         6=&FloatFromCurve(
          FloatCurve=DataInterface() {propertyValues {Curve {
             defaultValue=1
             keys [{interpMode=RCIM_Cubic, tangentMode=RCTM_User, time=0.82, value=1}, {interpMode=RCIM_Cubic, tangentMode=RCTM_User, time=1, value=0}]
            }}}
          Scale Curve=0.7
         )
         Index=&SpawnGroup()
         Num=::NewEnumerator5
        )
       )
      ))
    )
    Particles.GlowSize=&Multiply_Vector2D_ByFloat(
     Float=&Multiply_Float(
      A=&Multiply_Float(
       A=&MergeHelper_FloatSwitch(
        0=0
        1=0
        2=0
        3=0
        4=&FloatFromCurve(FloatCurve=DataInterface() {propertyValues {Curve {
            defaultValue=0
            keys [
             {arriveTangent=2.2222, time=0.0476, value=1}
             {arriveTangent=-1.1082, interpMode=RCIM_Cubic, leaveTangent=-1.1082, time=0.7, value=1}
             <at 0.95 0 RCIM_Cubic>
            ]
           }}})
        5=&FloatFromCurve(FloatCurve=DataInterface() {propertyValues {Curve {defaultValue=0, keys [{arriveTangent=2.2222, interpMode=RCIM_Constant, time=0.4468, value=0}, <at 0.7 1 RCIM_Constant>, <at 1 1 RCIM_Cubic>]}}})
        6=&FloatFromCurve(
         FloatCurve=DataInterface() {propertyValues {Curve {
            defaultValue=0
            keys [{arriveTangent=2.2222, interpMode=RCIM_Constant, time=0.4468, value=0}, <at 0.75 1 RCIM_Constant>, <at 1 1 RCIM_Cubic>]
           }}}
         Scale Curve=0.5
        )
        Index=&SpawnGroup()
        Num=::NewEnumerator5
       )
       B=26
      )
      B=$System.FrontendScale
     )
     Vector2D=&Vector2DFromFloat(Value=$Particles.HeartScale)
    )
   ) {isSetParameters=true}
  ]
  rendererClasses [&NiagaraSpriteRendererProperties]
  renderers [
   {rendererClass=&NiagaraSpriteRendererProperties}
   {rendererClass=&NiagaraSpriteRendererProperties, rendererIndex=1}
   {rendererClass=&NiagaraSpriteRendererProperties, rendererIndex=2}
  ]
 }
 LightRays {
  emitterUpdateScript [
   &EmitterState(Life Cycle Mode=::NewEnumerator1, Loop Behavior=::NewEnumerator1, Loop Duration=0.8, UseLoopDelay=false)
   &SpawnRate(SpawnGroup=1, SpawnRate=4, Use Spawn Probability=false) {module=SpawnRate001}
   &SpawnBurst_Instantaneous(Spawn Group=1, Use Spawn Probability=false, UseLoopCountLimit=false) {module=SpawnBurst_Instantaneous001}
   &SpawnRate(SpawnRate=9, Use Spawn Probability=false)
   &SpawnBurst_Instantaneous(Use Spawn Probability=false, UseLoopCountLimit=false)
  ]
  particleSpawnScript [
   &InitializeParticle(
    Color=&LerpLinearColors(StartColor=<rgba 0.1458 1.2325 2 1>, EndColor=<rgba 0.0217 0.3142 0.5208 3.5>, LerpFactor=0)
    Lifetime=&MergeHelper_FloatSwitch(0=0.6, 1=0.6, 2=?, 3=?, Index=&SpawnGroup(), Num=::NewEnumerator0)
    Position Offset=&Add_Vector(A=$System.BurstPosition, B=&FN_FrontEndVector3Modifier(Gameplay=<0 0 -61>, Front End=<0 0 -61>))
    Sprite Rotation Angle=&MergeHelper_FloatSwitch(
     0=0
     1=&RandomRangeFloat(Minimum=0, Maximum=1)
     2=?
     3=?
     Index=&SpawnGroup()
     Num=::NewEnumerator0
    )
    Sprite Rotation Mode=::NewEnumerator1
    Sprite Size=&Multiply_Vector2D_ByFloat(
     Float=$System.FrontendScale
     Vector2D=&RandomRangeVector2D(Minimum=<22 30>, Maximum=<24 36>, Evaluation Type=::NewEnumerator1)
    )
    Sprite Size Mode=::NewEnumerator4
    UsePositionOffset=true
    Write Material Random=true
   ) {module=InitializeParticle001}
   &SubUVAnimation(
    End Frame Range Override=3
    Evaluation Type=::NewEnumerator0
    "Sprite Renderer "=DataInterface() {propertyValues {SpriteRenderer=&self:LightRays_1.NiagaraSpriteRendererProperties_4}}
    SubUV Animation Mode:ENiagara_SubUVLookupModeV2=::NewEnumerator2
    UseEndFrame=true
    UseStartFrame=true
   )
   &SpriteFacingAndAlignment(Sprite Alignment=<0 0 1>, WriteAlignment=true, WriteFacing=true)
   /SetVariables(Particles.VisibilityTag=&SpawnGroup()) {isSetParameters=true}
  ]
  particleUpdateScript [
   &ParticleState()
   &ScaleSpriteSize(
    Scale Factor=&Vector2DFromCurve(
     Vector2Curve=DataInterface() {propertyValues {
       XCurve {keys [{arriveTangent=1.4704, interpMode=RCIM_Cubic, leaveTangent=1.4704, tangentMode=RCTM_User, time=0, value=0.7003}, <at 0.7026 2 RCIM_Cubic>]}
       YCurve {keys [
         {arriveTangent=0.3115, interpMode=RCIM_Cubic, leaveTangent=0.3115, tangentMode=RCTM_User, time=-0.0026, value=0.8481}
         {arriveTangent=0.1515, interpMode=RCIM_Cubic, leaveTangent=0.1515, time=0.3904, value=0.9697}
         <at 1 1 RCIM_Cubic>
        ]}
      }}
     ScaleCurve=<1 1>
    )
    Scale Sprite Size Mode=::NewEnumerator2
   )
   &CameraOffset(Camera Offset Amount=5)
   &SolveForcesAndVelocity(Clamp Velocity=false, Limit Acceleration=false)
   &ScaleColor(
    Scale Alpha=&FloatFromCurve(
     FloatCurve=DataInterface() {propertyValues {Curve {keys [<at 0 0 RCIM_Cubic>,<at 0.2498 1 RCIM_Cubic>,<at 1 0 RCIM_Cubic>]}}}
     Scale Curve=&FloatFromCurve(FloatCurve=DataInterface() {propertyValues {Curve {keys [<at 0.7 1>,<at 1.1 0>]}}}, CurveIndex=$System.Age)
    )
    ScaleA=true
    ScaleRGB=true
   )
   /SetVariables(
    Particles.FlareColor=&MakeLinearColorFromVectorAndFloat(
     "Vector (RGB)"=&MakeVectorFromLinearColorRGB(Color=&Multiply_LinearColorByFloat(
       Linear Color=<rgba 0.4896 1.3747 2 1>
       Float=&Waveform("Waveform [1]"=::NewEnumerator12, "[1] Frequency"=9, "[1] Amplitude Min/Max"=<0.4 1>)
      ))
     "Float (Alpha)"=&FloatFromCurve(
      FloatCurve=DataInterface() {propertyValues {Curve {keys [
          <at 0 0 RCIM_Cubic>
          <at 0.1957 1 RCIM_Cubic>
          {arriveTangent=-3.1851, interpMode=RCIM_Cubic, leaveTangent=-3.1851, tangentMode=RCTM_User, time=1, value=0}
         ]}}}
      Scale Curve=&FloatFromCurve(FloatCurve=DataInterface() {propertyValues {Curve {keys [<at 0.7 1>,<at 1.1 0>]}}}, CurveIndex=$System.Age)
     )
    )
    Particles.FlarePos:NiagaraPosition=&AddVectorToPosition(Position=$Particles.Position, Vector=<0 0 2.2>)
    Particles.FlareSize:Vector2f=&Vector2DFromFloat(Value=&Multiply_Float(
      A=&FloatFromCurve(
       FloatCurve=DataInterface() {propertyValues {Curve {keys [{arriveTangent=2.4801, interpMode=RCIM_Cubic, leaveTangent=2.4801, tangentMode=RCTM_User, time=0, value=0}, <at 1 1>]}}}
       Scale Curve=28
      )
      B=&Waveform("Waveform [1]"=::NewEnumerator12, "[1] Frequency"=9, "[1] Amplitude Min/Max"=<0.8 1>)
     ))
   ) {isSetParameters=true}
  ]
  rendererClasses [&NiagaraSpriteRendererProperties]
  renderers [
   {rendererClass=&NiagaraSpriteRendererProperties}
   {rendererClass=&NiagaraSpriteRendererProperties, rendererIndex=1}
  ]
 }
}
summary {
 systemName=NS_ReactFX_AutoMeow
 emitters [
  {emitterName=Heart, rendererClasses [&NiagaraSpriteRendererProperties]}
  {emitterName=Glitches, rendererClasses [&NiagaraMeshRendererProperties]}
  {emitterName=LightRays, rendererClasses [&NiagaraSpriteRendererProperties]}
 ]
 userVariables [
  User.bIsFrontEnd {type=NiagaraBool}
  User.bIsFrontEndPreview {type=NiagaraBool}
 ]
}
