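# Niagara system for the BamBoom reaction effect in Cosmetic Companions, playing a shared interaction sound with an animated sprite.
# The Heart sprite pops in with squash-and-stretch, rises, switches sub-UV frames and shakes later in life.
# A delayed burst throws spinning, shrinking debris from a shared system burst position.
# Front-end flags adjust offset and color, with time-of-day tinting in gameplay; two LWE emitters list no renderers.
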NiagaraSystem /CosmeticCompanions/FX/React/NS_ReactFX_BamBoom
dependencies {
 usedDataInterfaces [
  &NiagaraDataInterfaceAudioPlayer
  &NiagaraDataInterfaceCurve
  &NiagaraDataInterfaceVector2DCurve
  &NiagaraDataInterfaceSpriteRendererInfo
  &NiagaraDataInterfaceArrayFloat
 ]
 usedDynamicInputs [
  &MergeHelper_FloatSwitch
  &Lerp_Float
  &SpawnGroup
  &SelectFloatFromArray
  /Niagara/DynamicInputs/TypeConversions/MakeCustomPositionFromBool
 ]
 usedModules [
  &self:SystemSpawnScript.SystemScriptSource.SystemScriptGraph.NiagaraNodeAssignment_0.SetVariables_#1
  &self:SystemSpawnScript.SystemScriptSource.SystemScriptGraph.NiagaraNodeAssignment_1.SetVariables_#2
  &self:SystemSpawnScript.SystemScriptSource.SystemScriptGraph.NiagaraNodeAssignment_7.SetVariables_#3
  &self:SystemSpawnScript.SystemScriptSource.SystemScriptGraph.NiagaraNodeAssignment_4.SetVariables_#4
  &self:SystemSpawnScript.SystemScriptSource.SystemScriptGraph.NiagaraNodeAssignment_8.SetVariables_#5
  &self:SystemSpawnScript.SystemScriptSource.SystemScriptGraph.NiagaraNodeAssignment_2.SetVariables_#6
  &self:SystemSpawnScript.SystemScriptSource.SystemScriptGraph.NiagaraNodeAssignment_3.SetVariables_#7
  &self:SystemSpawnScript.SystemScriptSource.SystemScriptGraph.NiagaraNodeAssignment_9.SetVariables_#8
  &self:SystemSpawnScript.SystemScriptSource.SystemScriptGraph.NiagaraNodeAssignment_10.SetVariables_#9
  &self:SystemSpawnScript.SystemScriptSource.SystemScriptGraph.NiagaraNodeAssignment_5.SetVariables_#10
  &self:SystemSpawnScript.SystemScriptSource.SystemScriptGraph.NiagaraNodeAssignment_6.SetVariables_#11
  &SystemState
 ]
}
emitters {
 Debris001 {
  emitterUpdateScript [
   &EmitterState()
   &SpawnBurst_Instantaneous(
    Spawn Count=&RandomRangeInt(Minimum=8, Maximum=12)
    Spawn Time=0.925
    Use Spawn Probability=false
    UseLoopCountLimit=false
   )
  ]
  particleSpawnScript [
   &InitializeParticle(
    Color Mode=::NewEnumerator0
    Lifetime Max=0.75
    Lifetime Min=0.1
    Lifetime Mode=::NewEnumerator1
    Position Offset=$System.BurstPosition
    Sprite Rotation Mode=::NewEnumerator2
    Sprite Size Mode=::NewEnumerator1
    Uniform Sprite Size Max=12
    Uniform Sprite Size Min=6
    UsePositionOffset=true
    Write Material Random=true
   )
   &ShapeLocation(MaskSpawnGroup=false, Override Seed=false, Sphere Radius=2, UseOwnerScale=true)
   &AddVelocity(
    Module.Use Falloff Distance=false
    Override Seed=false
    Velocity Mode=::NewEnumerator1
    Velocity Origin=$Particles.ShapeLocation.ShapePosition
    Velocity Speed=&RandomRangeFloat(Minimum=80, Maximum=100)
   )
   &SubUVAnimation(
    Direct Frame Index:NiagaraFloat=&RandomRangeFloat(Minimum=0, Maximum=3)
    "Sprite Renderer "=DataInterface() {propertyValues {SpriteRenderer=&self:Debris001_0.NiagaraSpriteRendererProperties_0}}
    SubUV Animation Mode:ENiagara_SubUVLookupModeV2=::NewEnumerator5
   )
  ]
  particleUpdateScript [
   &ParticleState()
   &SpriteRotationRate(Rotation Rate=&RandomRangeFloat(Minimum=-1000, Maximum=1000))
   &ScaleSpriteSize(Uniform Curve Sprite Scale=DataInterface() {propertyValues {Curve {
       defaultValue=0
       keys [
        {arriveTangent=9.0725, interpMode=RCIM_Cubic, leaveTangent=9.0725, tangentMode=RCTM_User, time=0, value=0}
        {arriveTangent=-0.054, interpMode=RCIM_Cubic, leaveTangent=-0.054, tangentMode=RCTM_User, time=0.1408, value=0.9847}
        {arriveTangent=-0.0405, interpMode=RCIM_Cubic, leaveTangent=-0.0405, tangentMode=RCTM_User, time=1, value=0}
       ]
      }}})
   &GravityForce(Gravity=&RandomRangeVector(Minimum=<0 0 -100>, Maximum=<0 0 -250>))
   &Drag(Drag=&RandomRangeFloat(Minimum=0.8, Maximum=1.2), Use Linear Drag=true, Use Rotational Drag=true)
   &SolveForcesAndVelocity(Clamp Velocity=false, Limit Acceleration=false)
  ]
  rendererClasses [&NiagaraSpriteRendererProperties]
  renderers [{rendererClass=&NiagaraSpriteRendererProperties}]
 }
 Heart {
  emitterSpawnScript [
   &InitializeAudioPlayer(Audio Player=DataInterface() {propertyValues {
      Attenuation=null
      Concurrency=null
      MaxPlaysPerTick=10
      SoundToPlay=/CosmeticCompanions/Assets/Audio_Shared/Interaction/Petting/SC_Mimosa_Interaction_React_Default
      bAllowLoopingOneShotSounds=false
      bLimitPlaysPerTick=true
      bOnlyActiveDuringGameplay=false
      bStopWhenComponentIsDestroyed=true
      ConfigurationUserParameter {parameter {name=None, typeDefHandle {registeredTypeIndex=1}, varData []}}
      ParameterNames []
     }})
   /Niagara/Modules/Audio/PlayAudio() {module=PlayAudio001}
  ]
  emitterUpdateScript [
   &EmitterState()
   &SpawnBurst_Instantaneous(Use Spawn Probability=false, UseLoopCountLimit=false)
  ]
  particleSpawnScript [
   &InitializeParticle(
    Color=&MakeCustomLinearColorFromBool(
     Bool=$User.bIsFrontEndPreview
     True Color=<rgba 0.5 0.5 0.5 1>
     False Color=&FNTimeOfDayColor(
      Night Color=<rgba 0.5366 0.5587 0.625 1>
      Morning Color=<rgba 0.7188 0.649 0.577 1>
      Day Color=<rgba 0.8385 0.8385 0.8385 1>
      Evening Color=<rgba 0.7188 0.649 0.577 1>
     )
    )
    Position Offset=$System.BurstPosition
    Sprite Size Mode=::NewEnumerator3
    Uniform Sprite Size=28
    UsePositionOffset=false
    Write Material Random=true
   )
   &CameraOffset(Camera Offset Amount=10)
  ]
  particleUpdateScript [
   &ParticleState()
   &ScaleSpriteSize(Uniform Curve Sprite Scale=DataInterface() {propertyValues {Curve {keys [<at 0 0 RCIM_Cubic>,<at 0.225 1 RCIM_Cubic>]}}}) {enabled=false}
   &ScaleSpriteSize(
    Non-Uniform Curve Sprite Scale=DataInterface() {propertyValues {
      XCurve {keys [<at 0 0 RCIM_Cubic>,<at 0.15 1 RCIM_Cubic>]}
      YCurve {keys [
        <at 0 0 RCIM_Cubic>
        {arriveTangent=-0.033, interpMode=RCIM_Cubic, leaveTangent=-0.033, tangentMode=RCTM_User, time=0.075, value=1.75}
        <at 0.15 1 RCIM_Cubic>
       ]}
     }}
    Scale Sprite Size Mode=::NewEnumerator3
   ) {module=ScaleSpriteSize001}
   &SubUVAnimation(
    "Sprite Renderer "=DataInterface() {propertyValues {SpriteRenderer=&self:Heart_0.NiagaraSpriteRendererProperties_0}}
    SubUV Lookup Index:NiagaraFloat=&FloatFromCurve(FloatCurve=DataInterface() {propertyValues {Curve {keys [{arriveTangent=14.8245, interpMode=RCIM_Cubic, leaveTangent=14.8245, tangentMode=RCTM_User, time=0.15, value=0}, <at 0.25 1 RCIM_Cubic>]}}})
    UseEndFrame=false
    UseStartFrame=false
   )
   &OffsetPosition(Position Offset=&Lerp_Vector(
     A=&FN_FrontEndVector3Modifier(Gameplay=<0 0 -80>, Front End=<0 0 -40>)
     B=&FN_FrontEndVector3Modifier(Gameplay=<0 0 -15>, Front End=<0 0 -40>)
     Alpha=&FloatFromCurve(FloatCurve=DataInterface() {propertyValues {Curve {
         defaultValue=1
         keys [
          {arriveTangent=7.9822, interpMode=RCIM_Cubic, leaveTangent=7.9822, tangentMode=RCTM_User, time=0, value=0}
          {interpMode=RCIM_Cubic, tangentMode=RCTM_User, time=0.225, value=1}
         ]
        }}})
    ))
   &ScaleColor(
    Scale Alpha=&FloatFromCurve(FloatCurve=DataInterface() {propertyValues {Curve {keys [<at 0 1>,<at 0.8938 1>,<at 1 0>]}}})
    Scale RGB=&VectorFromFloat(Value=2.25)
    ScaleA=true
    ScaleRGB=true
   )
   /SetVariables(Particles.Position=&AddVectorToPosition(
     Position=$Particles.OffsetPosition.Position
     Vector=&MakeVector(
      X=&Waveform(
       "Waveform [1]"=::NewEnumerator0
       "[1] Frequency"=&FloatFromCurve(FloatCurve=DataInterface() {propertyValues {Curve {keys [<at 0.6 0>,<at 0.7 1>]}}}, Scale Curve=8)
       "[1] Amplitude Min/Max"=<-1 1>
      )
      Y=0
      Z=&Waveform(
       "Waveform [1]"=::NewEnumerator0
       "[1] Frequency"=&FloatFromCurve(FloatCurve=DataInterface() {propertyValues {Curve {keys [<at 0.6 0>,<at 0.7 1>]}}}, Scale Curve=8)
       "[1] Amplitude Min/Max"=<-1 1>
      )
     )
    )) {isSetParameters=true}
  ]
  rendererClasses [&NiagaraSpriteRendererProperties]
  renderers [{rendererClass=&NiagaraSpriteRendererProperties}]
 }
 LWE_Corona001 {rendererClasses [], renderers []}
 LWE_Rays {rendererClasses [], renderers []}
}
summary {
 systemName=NS_ReactFX_BamBoom
 emitters [
  {emitterName=LWE_Rays, rendererClasses []}
  {emitterName=Heart, rendererClasses [&NiagaraSpriteRendererProperties]}
  {emitterName=LWE_Corona001, rendererClasses []}
  {emitterName=Debris001, rendererClasses [&NiagaraSpriteRendererProperties]}
 ]
 userVariables [
  User.bIsFrontEnd {type=NiagaraBool}
  User.bIsFrontEndPreview {type=NiagaraBool, defaultValue=true}
 ]
}
