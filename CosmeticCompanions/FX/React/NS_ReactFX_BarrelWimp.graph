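# Niagara reaction effect for the BarrelWimp cosmetic companion, with a petting-reaction sound played through the Intro emitter.
# Intro sprites, glow and ink smoke travel from HeartIntroSpawnPosition to BurstPosition, while a purple mesh smoke ring swirls near the start.
# Pop, glow, smoke and radial lightning-mesh bursts are timed off System.IntroLength to land at the burst point.
# Front-end modifiers adjust scale and brightness, and some emitters also dim by time of day.
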NiagaraSystem /CosmeticCompanions/FX/React/NS_ReactFX_BarrelWimp
dependencies {
 usedDataInterfaces [
  &NiagaraDataInterfaceAudioPlayer
  &NiagaraDataInterfaceCurve
  &NiagaraDataInterfaceVectorCurve
  &NiagaraDataInterfaceVector2DCurve
  &NiagaraDataInterfaceSpriteRendererInfo
  &NiagaraDataInterfaceColorCurve
  &NiagaraDataInterfaceArrayFloat
 ]
 usedDynamicInputs [
  &Vector2DFromCurve
  &RandomRangeVector2D
  &Vector2DFromFloat
 ]
 usedModules [
  &self:SystemSpawnScript.SystemScriptSource.SystemScriptGraph.NiagaraNodeAssignment_0.SetVariables_#1
  &self:SystemSpawnScript.SystemScriptSource.SystemScriptGraph.NiagaraNodeAssignment_6.SetVariables_#2
  &self:SystemSpawnScript.SystemScriptSource.SystemScriptGraph.NiagaraNodeAssignment_7.SetVariables_#3
  &self:SystemSpawnScript.SystemScriptSource.SystemScriptGraph.NiagaraNodeAssignment_5.SetVariables_#4
  &self:SystemSpawnScript.SystemScriptSource.SystemScriptGraph.NiagaraNodeAssignment_2.SetVariables_#5
  &self:SystemSpawnScript.SystemScriptSource.SystemScriptGraph.NiagaraNodeAssignment_3.SetVariables_#6
  &SystemState
 ]
}
emitters {
 Glow_n_Pop {
  emitterUpdateScript [
   &EmitterState()
   &SpawnBurst_Instantaneous(Spawn Time=&Subtract_Float(A=$System.IntroLength, B=0), Use Spawn Probability=false, UseLoopCountLimit=false)
   &SpawnBurst_Instantaneous(Spawn Group=1, Spawn Time=0.05, Use Spawn Probability=false, UseLoopCountLimit=false) {module=SpawnBurst_Instantaneous001}
  ]
  particleSpawnScript [
   &InitializeParticle(
    Color=&MaskLinearColorBySpawnGroup(Masked Linear Color=<rgba 2.7355 0.3664 3 0.8>, Passthrough Linear Color=<rgba 2.679 0.8015 3 0.8>)
    Lifetime=&MaskFloatBySpawnGroup(Masked Float=0.6, Passthrough Float=1)
    Position Offset=$System.BurstPosition
    Sprite Rotation Mode=::NewEnumerator0
    Sprite Size Mode=::NewEnumerator3
    Uniform Sprite Size=&Multiply_Float(
     A=&MaskFloatBySpawnGroup(Masked Float=30, Passthrough Float=30)
     B=&FN_FrontEndFloatModifier(Gameplay=1, Front End=$System.FrontendScale)
    )
    UsePositionOffset=true
    Write Material Random=true
   )
   &DynamicMaterialParameters(
    Index 0 Param 4=1
    Module.Index0Float0Write=true
    Module.Index0Float1Write=true
    Module.Index0Float2Write=true
    Module.Index0Float3Write=true
   )
   /SetVariables(Particles.VisibilityTag=&SpawnGroup()) {isSetParameters=true}
   &CameraOffset(Camera Offset Amount=&MaskFloatBySpawnGroup(Masked Float=0, Passthrough Float=10)) {enabled=false}
  ]
  particleUpdateScript [
   &ParticleState()
   &ScaleSpriteSize(
    Scale Sprite Size Mode=::NewEnumerator0
    Uniform Scale Factor=&FloatFromCurve(FloatCurve=DataInterface() {propertyValues {Curve {keys [
         {arriveTangent=12.8113, interpMode=RCIM_Cubic, leaveTangent=12.8113, tangentMode=RCTM_User, time=0, value=0}
         {arriveTangent=3.3458, interpMode=RCIM_Cubic, leaveTangent=3.3458, tangentMode=RCTM_User, time=0.1, value=1}
         {arriveTangent=-0.0026, tangentMode=RCTM_User, time=1, value=3}
        ]}}})
   ) {module=ScaleSpriteSize002}
   &ScaleColor(
    Scale Alpha=&FloatFromCurve(FloatCurve=DataInterface() {propertyValues {Curve {
        defaultValue=0
        keys [
         {arriveTangent=0.3137, interpMode=RCIM_Cubic, leaveTangent=0.3137, tangentMode=RCTM_User, time=0, value=1}
         {arriveTangent=-2.0958, interpMode=RCIM_Cubic, leaveTangent=-2.0958, tangentMode=RCTM_User, time=1, value=0}
        ]
       }}})
    Scale RGB=&VectorFromCurve(
     VectorCurve=DataInterface() {propertyValues {
       XCurve {keys [{interpMode=RCIM_Cubic, tangentMode=RCTM_User, time=0, value=1}, {arriveTangent=-2.0498, tangentMode=RCTM_User, time=1, value=0}]}
       YCurve {keys [{interpMode=RCIM_Cubic, tangentMode=RCTM_User, time=0, value=1}, {arriveTangent=-2.0498, tangentMode=RCTM_User, time=1, value=0}]}
       ZCurve {keys [{interpMode=RCIM_Cubic, tangentMode=RCTM_User, time=0, value=1}, {arriveTangent=-2.0498, tangentMode=RCTM_User, time=1, value=0}]}
      }}
     Scale Curve=&MakeVectorFromLinearColorRGB(Color=$System.TOD)
    )
    ScaleA=true
    ScaleRGB=true
   )
   &FN_Sprite_Attributes(
    Color=&MultiplyLinearColor(
     Linear Color=$Particles.Initial.Color
     Scale Factor=&MakeVector4FromVectorAndFloat(
      "Vector (RGB)"=<1 1 1>
      "Float (Alpha)"=&FloatFromCurve(FloatCurve=DataInterface() {propertyValues {Curve {
          defaultValue=1
          keys [
           {arriveTangent=2.6956, interpMode=RCIM_Cubic, leaveTangent=2.6956, tangentMode=RCTM_User, time=0.875, value=1}
           {arriveTangent=-35.2676, interpMode=RCIM_Cubic, leaveTangent=-35.2676, tangentMode=RCTM_User, time=0.9, value=0}
          ]
         }}})
     )
    )
    Create Color Attribute=true
    Create Position Attribute=true
    Create SpriteSize Attribute=true
    SpriteSize=&Multiply_Vector2D_ByFloat(
     Float=&FloatFromCurve(FloatCurve=DataInterface() {propertyValues {Curve {
         defaultValue=1
         keys [
          {arriveTangent=15.6628, interpMode=RCIM_Cubic, leaveTangent=15.6628, tangentMode=RCTM_User, time=0, value=0}
          {arriveTangent=1.8237, interpMode=RCIM_Cubic, leaveTangent=1.8237, tangentMode=RCTM_User, time=0.1, value=1}
          {interpMode=RCIM_Cubic, tangentMode=RCTM_User, time=0.8, value=1.6}
         ]
        }}})
     Vector2D=$Particles.Initial.SpriteSize
    )
    Write to Module Namespace=true
   )
   /SetVariables(Particles.Glow Color=&Multiply_LinearColorByFloat(
     Linear Color=$Particles.Color
     Float=&FloatFromCurve(FloatCurve=DataInterface() {propertyValues {Curve {
         defaultValue=1
         keys [
          {arriveTangent=23.2094, interpMode=RCIM_Cubic, leaveTangent=23.2094, tangentMode=RCTM_User, time=0, value=0.2}
          {interpMode=RCIM_Cubic, tangentMode=RCTM_User, time=0.1, value=1}
          {arriveTangent=-2.376, interpMode=RCIM_Cubic, leaveTangent=-2.376, tangentMode=RCTM_User, time=1, value=0}
         ]
        }}})
    )) {isSetParameters=true, module=SetVariables_#7}
   &OffsetPosition(Position Offset=&Lerp_Vector(
     A=$System.HeartIntroSpawnPosition
     B=$System.BurstPosition
     Alpha=&FloatFromCurve(FloatCurve=DataInterface() {propertyValues {Curve {
         defaultValue=0
         keys [
          {arriveTangent=4.1785, interpMode=RCIM_Cubic, leaveTangent=4.1785, tangentMode=RCTM_User, time=0, value=0}
          {arriveTangent=1.1179, interpMode=RCIM_Cubic, leaveTangent=1.1179, tangentMode=RCTM_User, time=0.2042, value=0.6228}
          {arriveTangent=0.0224, interpMode=RCIM_Cubic, leaveTangent=0.0224, tangentMode=RCTM_User, time=1, value=1}
         ]
        }}})
    ))
  ]
  rendererClasses [&NiagaraSpriteRendererProperties]
  renderers [
   {rendererClass=&NiagaraSpriteRendererProperties}
   {rendererClass=&NiagaraSpriteRendererProperties, rendererIndex=1}
   {rendererClass=&NiagaraSpriteRendererProperties, rendererIndex=2}
  ]
 }
 InkSmoke {
  emitterUpdateScript [
   &EmitterState(Life Cycle Mode=::NewEnumerator1, Loop Behavior=::NewEnumerator1, Loop Delay=0.4, Loop Duration=2, UseLoopDelay=false)
   &SpawnRate(
    SpawnRate=&FloatFromCurve(
     FloatCurve=DataInterface() {propertyValues {Curve {keys [{arriveTangent=3.0606, interpMode=RCIM_Constant, time=0, value=0}, <at 0.1 1 RCIM_Constant>, <at 0.6 0 RCIM_Constant>]}}}
     CurveIndex=&Divide_Float(A=$Emitter.Age, B=$System.IntroLength)
     Scale Curve=20
    )
    Use Spawn Probability=false
   )
   &SpawnBurst_Instantaneous(
    Spawn Count=5
    Spawn Group=1
    Spawn Time=&Add_Float(A=$System.IntroLength, B=-0.2)
    Use Spawn Probability=false
    UseLoopCountLimit=false
   )
  ]
  particleSpawnScript [
   &InitializeParticle(
    Lifetime=&MergeHelper_FloatSwitch(
     0=&RandomRangeFloat(Minimum=0.8, Maximum=1.2)
     1=&RandomRangeFloat(Minimum=0.5, Maximum=1)
     2=?
     3=?
     Index=&SpawnGroup()
     Num=::NewEnumerator0
    )
    Position Offset=$System.BurstPosition
    Sprite Rotation Mode=::NewEnumerator2
    Sprite Size Mode=::NewEnumerator3
    Uniform Sprite Size=&Multiply_Float(
     A=&MergeHelper_FloatSwitch(
      0=&RandomRangeFloat(Minimum=20, Maximum=30)
      1=&RandomRangeFloat(Minimum=35, Maximum=40)
      2=?
      3=?
      Index=&SpawnGroup()
      Num=::NewEnumerator0
     )
     B=&FN_FrontEndFloatModifier(Gameplay=1, Front End=$User.FrontEndScale)
    )
    UsePositionOffset=false
    Write Material Random=true
   )
   &CameraOffset(Camera Offset Amount=30)
   &ShapeLocation(MaskSpawnGroup=false, Override Seed=false, Sphere Radius=11, UseOwnerScale=true)
  ]
  particleUpdateScript [
   &ParticleState()
   &ScaleVelocity(Velocity Scale=&VectorFromCurve(VectorCurve=DataInterface() {propertyValues {
       XCurve {keys [{arriveTangent=-1.756, interpMode=RCIM_Cubic, leaveTangent=-1.756, tangentMode=RCTM_User, time=0, value=1}, <at 1 0>]}
       YCurve {keys [{arriveTangent=-1.756, interpMode=RCIM_Cubic, leaveTangent=-1.756, tangentMode=RCTM_User, time=0, value=1}, <at 1 0>]}
       ZCurve {keys [{arriveTangent=-1.756, interpMode=RCIM_Cubic, leaveTangent=-1.756, tangentMode=RCTM_User, time=0, value=1}, <at 1 0>]}
      }}))
   &SpriteRotationRate(Rotation Rate=&RandomRangeFloat(Minimum=-50, Maximum=50))
   &OffsetPosition(Position Offset=&Lerp_Vector(
     A=$System.HeartIntroSpawnPosition
     B=$System.BurstPosition
     Alpha=&FloatFromCurve(
      FloatCurve=DataInterface() {propertyValues {Curve {keys [
          {arriveTangent=3.0606, interpMode=RCIM_Cubic, leaveTangent=3.0606, tangentMode=RCTM_User, time=0, value=0}
          {arriveTangent=-0.1073, tangentMode=RCTM_User, time=1, value=1}
         ]}}}
      CurveIndex=&Divide_Float(A=$Emitter.Age, B=$System.IntroLength)
     )
    ))
   &ScaleSpriteSize(
    Scale Sprite Size Mode=::NewEnumerator0
    Uniform Scale Factor=&MergeHelper_FloatSwitch(
     0=&FloatFromCurve(
      FloatCurve=DataInterface() {propertyValues {Curve {
         defaultValue=1
         keys [
          {arriveTangent=4.1785, interpMode=RCIM_Cubic, leaveTangent=4.1785, tangentMode=RCTM_User, time=0, value=0}
          {arriveTangent=1.1179, interpMode=RCIM_Cubic, leaveTangent=1.1179, tangentMode=RCTM_User, time=0.5, value=1}
          {arriveTangent=0.0224, interpMode=RCIM_Cubic, leaveTangent=0.0224, tangentMode=RCTM_User, time=1, value=2}
         ]
        }}}
      Scale Curve=0.5
     )
     1=&FloatFromCurve(
      FloatCurve=DataInterface() {propertyValues {Curve {
         defaultValue=1
         keys [
          {arriveTangent=4.1785, interpMode=RCIM_Cubic, leaveTangent=4.1785, tangentMode=RCTM_User, time=0, value=0}
          {arriveTangent=0.0224, interpMode=RCIM_Cubic, leaveTangent=0.0224, tangentMode=RCTM_User, time=1, value=2}
         ]
        }}}
      Scale Curve=0.5
     )
     2=?
     3=?
     Index=&SpawnGroup()
     Num=::NewEnumerator0
    )
   )
   &SubUVAnimation(
    End Frame Range Override=30
    "Sprite Renderer "=DataInterface() {propertyValues {SpriteRenderer=&self:InkSmoke_0.NiagaraSpriteRendererProperties_0}}
    Start Frame Range Override=&RandomRangeInt(Minimum=0, Maximum=5)
    UseEndFrame=true
    UseStartFrame=true
   )
   &ScaleColor(
    Scale Alpha=&Multiply_Float(
     A=&RandomRangeFloat(Minimum=0.5, Maximum=1)
     B=&FloatFromCurve(
      FloatCurve=DataInterface() {propertyValues {Curve {keys [{interpMode=RCIM_Cubic, tangentMode=RCTM_User, time=0.5, value=1}, {tangentMode=RCTM_User, time=1, value=0}]}}}
      Scale Curve=&FN_FrontEndFloatModifier(Gameplay=1, Front End=1.5)
     )
    )
    Scale RGB=&MergeHelper_VectorSwitch(
     0=&MakeVectorFromLinearColorRGB(Color=<rgba 0.6277 0.2835 0.7487 1>)
     1=&MakeVectorFromLinearColorRGB(Color=&ColorFromCurve(
       ColorCurve=DataInterface() {propertyValues {
         AlphaCurve {keys [<at 0 1 RCIM_Cubic>,<at 1 0 RCIM_Cubic>]}
         BlueCurve {keys [<at 0 0.8006 RCIM_Cubic>,<at 0.4459 0.1733 RCIM_Cubic>]}
         GreenCurve {keys [<at 0 0.2767 RCIM_Cubic>,<at 0.4459 0 RCIM_Cubic>]}
         RedCurve {keys [<at 0 0.6849 RCIM_Cubic>,<at 0.4459 0.0832 RCIM_Cubic>]}
        }}
       CurveIndex=&Add_Float(A=$Particles.Age, B=&RandomRangeFloat(Minimum=0.1, Maximum=0))
      ))
     2=?
     3=?
     Index=&SpawnGroup()
     Num=::NewEnumerator0
    )
    ScaleA=true
    ScaleRGB=true
   )
   &SolveForcesAndVelocity(Clamp Velocity=false, Limit Acceleration=false)
  ]
  rendererClasses [&NiagaraSpriteRendererProperties]
  renderers [{rendererClass=&NiagaraSpriteRendererProperties}]
 }
 Intro {
  emitterSpawnScript [
   &InitializeAudioPlayer(Audio Player=DataInterface() {propertyValues {
      Attenuation=null
      Concurrency=null
      MaxPlaysPerTick=10
      SoundToPlay=/CosmeticCompanions/Assets/Audio_Shared/Interaction/Petting/SC_Mimosa_Interaction_React_Default
      bAllowLoopingOneShotSounds=false
      bLimitPlaysPerTick=true
      bOnlyActiveDuringGameplay=false
      bStopWhenComponentIsDestroyed=true
      ConfigurationUserParameter {parameter {name=None, typeDefHandle {registeredTypeIndex=1}, varData []}}
      ParameterNames []
     }})
   /Niagara/Modules/Audio/PlayAudio() {module=PlayAudio001}
  ]
  emitterUpdateScript [
   &EmitterState()
   &SpawnRate(
    SpawnGroup=1
    SpawnRate=&FloatFromCurve(FloatCurve=DataInterface() {propertyValues {Curve {keys [<at 0.49 1>,<at 0.5 0>]}}}, CurveIndex=$Emitter.NormalizedLoopAge, Scale Curve=50)
    Use Spawn Probability=false
   )
   &SpawnBurst_Instantaneous(Spawn Time=0.2, Use Spawn Probability=false, UseLoopCountLimit=false)
  ]
  particleSpawnScript [
   /SetVariables(Particles.VisibilityTag=$Emitter.SpawnGroup) {isSetParameters=true}
   &InitializeParticle(
    Color=<rgba 0 352.0194 500 1>
    Lifetime=0.9
    Ribbon Width=10
    Ribbon Width Mode=::NewEnumerator1
    Sprite Rotation Mode=::NewEnumerator0
    Sprite Size Mode=::NewEnumerator3
    Uniform Sprite Size=&Multiply_Float(A=18, B=&FN_FrontEndFloatModifier(Gameplay=1, Front End=$User.FrontEndScale))
    UsePositionOffset=false
    Write Material Random=true
   )
  ]
  particleUpdateScript [
   &ParticleState()
   &ScaleSpriteSize(
    Scale Sprite Size Mode=::NewEnumerator0
    Uniform Scale Factor=&Add_Float(
     A=0
     B=&FloatFromCurve(FloatCurve=DataInterface() {propertyValues {Curve {
         defaultValue=0
         keys [
          <at 0 0.3 RCIM_Cubic>
          {arriveTangent=1.25, leaveTangent=1.25, time=0.4, value=1}
          {arriveTangent=0.0079, interpMode=RCIM_Cubic, leaveTangent=0.0079, tangentMode=RCTM_User, time=0.8153, value=1.0094}
          {arriveTangent=14.0363, interpMode=RCIM_Cubic, leaveTangent=14.0363, tangentMode=RCTM_User, time=1, value=2}
         ]
        }}})
    )
   ) {module=ScaleSpriteSize002}
   &ScaleColor(
    Scale Alpha=&FloatFromCurve(FloatCurve=DataInterface() {propertyValues {Curve {
        defaultValue=0
        keys [
         {arriveTangent=9.0725, interpMode=RCIM_Cubic, leaveTangent=9.0725, tangentMode=RCTM_User, time=0, value=0}
         {arriveTangent=-0.2441, leaveTangent=-0.2441, time=0.1408, value=0.9847}
         {arriveTangent=-1.146, leaveTangent=-1.146, time=0.95, value=1}
         {arriveTangent=-0.0405, interpMode=RCIM_Cubic, leaveTangent=-0.0405, tangentMode=RCTM_User, time=1, value=0}
        ]
       }}})
    ScaleA=true
    ScaleRGB=true
   )
   &OffsetPosition(Position Offset=&Lerp_Vector(
     A=$System.HeartIntroSpawnPosition
     B=$System.BurstPosition
     Alpha=&FloatFromCurve(FloatCurve=DataInterface() {propertyValues {Curve {
         defaultValue=0
         keys [
          {arriveTangent=4.1785, interpMode=RCIM_Cubic, leaveTangent=4.1785, tangentMode=RCTM_User, time=0, value=0}
          {arriveTangent=1.1179, interpMode=RCIM_Cubic, leaveTangent=1.1179, tangentMode=RCTM_User, time=0.2042, value=0.6228}
          {arriveTangent=0.0224, interpMode=RCIM_Cubic, leaveTangent=0.0224, tangentMode=RCTM_User, time=1, value=1}
         ]
        }}})
    ))
   &DynamicMaterialParameters(
    Index 0 Param 1=&FloatFromCurve(FloatCurve=DataInterface() {propertyValues {Curve {keys [{arriveTangent=0.6296, interpMode=RCIM_Cubic, leaveTangent=0.6296, tangentMode=RCTM_User, time=0.5, value=1}, <at 1 2>]}}})
    Index 0 Param 2=&FloatFromCurve(FloatCurve=DataInterface() {propertyValues {Curve {keys [{arriveTangent=0.6296, interpMode=RCIM_Cubic, leaveTangent=0.6296, tangentMode=RCTM_User, time=0.5, value=1}, <at 0.6 50>, <at 0.7 1>]}}})
    Index 0 Param 3=1
    Module.Index0Float0Write=true
    Module.Index0Float1Write=true
    Module.Index0Float2Write=true
    Module.Index0Float3Write=true
   )
   /SetVariables(
    Particles.Glow Color=&Multiply_LinearColorByFloat(Linear Color=<rgba 0.2107 0.072 0.4427 0.5>, Float=0.8)
    Particles.Glow Size=&Multiply_Vector2D(A=$Particles.SpriteSize, B=<1.5 1.5>)
   ) {isSetParameters=true}
   &ScaleRibbonWidth(Ribbon Width Scale=&FloatFromCurve(FloatCurve=DataInterface() {propertyValues {Curve {keys [<at 0 1>,<at 0.2 1>,<at 0.7 1>,<at 1 0>]}}}))
  ]
  rendererClasses [&NiagaraSpriteRendererProperties]
  renderers [
   {rendererClass=&NiagaraSpriteRendererProperties}
   {rendererClass=&NiagaraSpriteRendererProperties, rendererIndex=1}
  ]
 }
 Lightning001 {
  emitterUpdateScript [
   &EmitterState()
   &SpawnBurst_Instantaneous(
    Spawn Count=3
    Spawn Group=1
    Spawn Time=&Subtract_Float(A=$System.IntroLength, B=0.3)
    Use Spawn Probability=false
    UseLoopCountLimit=false
   ) {module=SpawnBurst_Instantaneous001}
   &SpawnBurst_Instantaneous(
    Spawn Count=&RandomRangeInt(Minimum=4, Maximum=5)
    Spawn Time=&Subtract_Float(A=$System.IntroLength, B=0.1)
    Use Spawn Probability=false
    UseLoopCountLimit=false
   )
  ]
  particleSpawnScript [
   &PartitionParticles()
   &InitializeParticle(
    Color=&MaskLinearColorBySpawnGroup(Masked Linear Color=<rgba 1 0 0.9481 1>, Passthrough Linear Color=<rgba 1 0.2895 0.9763 1>)
    Lifetime Max=0.55
    Lifetime Min=0.3
    Lifetime Mode=::NewEnumerator1
    Mesh Scale=&RandomRangeVector(Minimum=<0.3 0.5 0.5>, Maximum=<0.4 0.5 0.5>)
    Mesh Scale Mode=::NewEnumerator4
    Position Offset=$System.BurstPosition
    UsePositionOffset=true
    Write Material Random=true
   )
   &CameraOffset(Camera Offset Amount=10)
   &ShapeLocation(MaskSpawnGroup=false, Override Seed=false, Sphere Radius=1, UseOwnerScale=true)
   &InitialMeshOrientation(
    Mesh Orientation Mode=::NewEnumerator2
    Orientation Vector=&Add_Vector(
     A=/Niagara/DynamicInputs/Vectors/RotateVector(
      Vector To Rotate:Vector3f=<1 0 0>
      Yaw:NiagaraFloat=&ReturnNormalizedExecIndex(Normalized Index Includes One=false)
      Pitch:NiagaraFloat=0
      Roll:NiagaraFloat=0
      Rotation Rate:NiagaraFloat=360
     )
     B=&MakeVector(
      X=0
      Y=0
      Z=&Multiply_Float(
       A=&RandomRangeFloat(Minimum=0.5, Maximum=1)
       B=&SelectFloatFromArray(
        Direct Array Index=&ModuloInt(A=&ReturnExecIndex(), B=2)
        Float Selection Array=DataInterface() {propertyValues {FloatData [-1,1]}}
       )
      )
     )
    )
    Use Rotation Vector=false
   ) {module=InitialMeshOrientation002}
   /SetVariables(Particles.Random=&RandomRangeFloat(Minimum=0, Maximum=1)) {isSetParameters=true}
  ]
  particleUpdateScript [
   &ParticleState()
   &ScaleMeshSize(Scale Factor=&Multiply_VectorByFloat(
     Vector=&FN_FrontEndVector3Modifier(Gameplay=<1 1 1>, Front End=<0.8 1 1>)
     Float=&MaskFloatBySpawnGroup(Masked Float=0.7, Passthrough Float=0.35)
    ))
   &ScaleColor(
    Scale Alpha=&Multiply_Float(
     A=10
     B=&FloatFromCurve(FloatCurve=DataInterface() {propertyValues {Curve {keys [<at 0 0>,<at 0.1 0>,<at 0.2 0.8>,<at 1 0>]}}})
    )
    Scale RGB=&Multiply_VectorByFloat(
     Vector=&Multiply_VectorByFloat(Vector=<1 1 1>, Float=&FN_FrontEndFloatModifier(Gameplay=0.5, Front End=0.5))
     Float=&Multiply_Float(
      A=&RandomRangeFloat(Minimum=0, Maximum=1, Evaluation Type=::NewEnumerator1)
      B=&Multiply_Float(A=&FNTimeOfDayScalar(Night Scale=0.2, Morning Scale=1, Evening Scale=1), B=10)
     )
    )
    ScaleA=true
    ScaleRGB=true
   )
   &DynamicMaterialParameters(
    Index 0 Param 1=&FloatFromCurve(FloatCurve=DataInterface() {propertyValues {Curve {keys [{tangentMode=RCTM_User, time=0, value=1}, {tangentMode=RCTM_User, time=1, value=1}]}}})
    Index 0 Param 2=0.7
    Index 0 Param 3=&FloatFromCurve(FloatCurve=DataInterface() {propertyValues {Curve {keys [
         {arriveTangent=4.0072, interpMode=RCIM_Cubic, leaveTangent=4.0072, tangentMode=RCTM_User, time=0, value=0}
         {arriveTangent=2.7083, interpMode=RCIM_Cubic, leaveTangent=2.7083, time=0.2543, value=0.6225}
         {arriveTangent=1.8473, interpMode=RCIM_Cubic, leaveTangent=1.8473, time=0.6, value=1.625}
         {arriveTangent=1.6993, interpMode=RCIM_Cubic, leaveTangent=1.6993, tangentMode=RCTM_User, time=1, value=2}
        ]}}})
    Index 0 Param 4=5
    Index 1 Param 1=&FN_FrontEndFloatModifier(Gameplay=1.2, Front End=0.5)
    Index 1 Param 2=$Particles.Random
    Index 1 Param 3=1
    Index 1 Param 4=1
    Index 2 Param 1=&FloatFromCurve(FloatCurve=DataInterface() {propertyValues {Curve {keys [<at 0 1>,<at 1 0.3>]}}})
    Module.Param0WriteEnabled=true
    Module.Param0WriteEnabled1=true
    Module.Param0WriteEnabled2=true
    Module.Param1WriteEnabled=true
    Module.Param1WriteEnabled1=true
    Module.Param1WriteEnabled2=true
    Module.Param2WriteEnabled=true
    Module.Param2WriteEnabled1=true
    Module.Param2WriteEnabled2=true
    Module.Param3WriteEnabled=true
    Module.Param3WriteEnabled1=true
    Module.Param3WriteEnabled2=true
    Write Parameter Index 1=true
    Write Parameter Index 2=true
   )
   &SolveForcesAndVelocity(Clamp Velocity=false, Limit Acceleration=false)
   /SetVariables(Particles.Lightning glow color:LinearColor=&MakeLinearColorFromVectorAndFloat(
     "Vector (RGB)"=&MakeVectorFromLinearColorRGB(Color=&MultiplyLinearColor(
       Linear Color=&MaskLinearColorBySpawnGroup(Masked Linear Color=<rgba 1 0.2895 0.9763 1>, Passthrough Linear Color=<rgba 1 0 0.9481 1>)
       Scale Factor=&MakeVector4_FromFloat(Float=&RandomRangeFloat(Minimum=0.5, Maximum=1, Evaluation Type=::NewEnumerator1))
      ))
     "Float (Alpha)"=&Multiply_Float(
      A=&FloatFromCurve(FloatCurve=DataInterface() {propertyValues {Curve {keys [<at 0 1 RCIM_Cubic>, {interpMode=RCIM_Cubic, leaveTangent=-3.6307, time=0.9, value=0}]}}})
      B=7
     )
    )) {isSetParameters=true}
   /SetVariables(Particles.Glow Dy02:Vector4f=&MakeVector4(
     X=&FloatFromCurve(FloatCurve=DataInterface() {propertyValues {Curve {keys [{tangentMode=RCTM_User, time=0, value=0.2}, {tangentMode=RCTM_User, time=0.5, value=1}]}}})
     Y=0.7
     Z=0
     W=5
    )) {isSetParameters=true}
   /SetVariables(Particles.Glow Dy03:Vector4f=&MakeVector4(
     X=&FloatFromCurve(
      FloatCurve=DataInterface() {propertyValues {Curve {keys [<at 0 1>,<at 1 0.5>]}}}
      Scale Curve=&FN_FrontEndFloatModifier(Gameplay=1.5, Front End=1)
     )
     Y=$Particles.Random
     Z=1
     W=1
    )) {isSetParameters=true}
  ]
  rendererClasses [&NiagaraMeshRendererProperties]
  renderers [
   {rendererClass=&NiagaraMeshRendererProperties, rendererIndex=0}
   {rendererClass=&NiagaraMeshRendererProperties, rendererIndex=1}
  ]
 }
 Pop {
  emitterUpdateScript [
   &EmitterState()
   &SpawnBurst_Instantaneous(
    Spawn Time=&Subtract_Float(A=$System.IntroLength, B=0.1)
    Use Spawn Probability=false
    UseLoopCountLimit=false
   )
  ]
  particleSpawnScript [
   &InitializeParticle(
    Color=<rgba 5 1.0938 4.4808 0.8>
    Lifetime=0.2
    Position Offset=$System.BurstPosition
    Sprite Rotation Mode=::NewEnumerator0
    Sprite Size Mode=::NewEnumerator3
    Uniform Sprite Size=&Multiply_Float(A=20, B=&FN_FrontEndFloatModifier(Gameplay=0.7, Front End=$User.FrontEndScale))
    UsePositionOffset=true
    Write Material Random=true
   )
   &DynamicMaterialParameters(
    Index 0 Param 4=1
    Module.Index0Float0Write=true
    Module.Index0Float1Write=true
    Module.Index0Float2Write=true
    Module.Index0Float3Write=true
   )
  ]
  particleUpdateScript [
   &ParticleState()
   &ScaleSpriteSize(
    Scale Sprite Size Mode=::NewEnumerator0
    Uniform Scale Factor=&FloatFromCurve(FloatCurve=DataInterface() {propertyValues {Curve {keys [
         {arriveTangent=3.3458, interpMode=RCIM_Cubic, leaveTangent=3.3458, tangentMode=RCTM_User, time=0, value=1}
         {arriveTangent=1.251, tangentMode=RCTM_User, time=1, value=3}
        ]}}})
   ) {module=ScaleSpriteSize002}
   &ScaleColor(
    Scale Alpha=&FloatFromCurve(FloatCurve=DataInterface() {propertyValues {Curve {
        defaultValue=0
        keys [
         {arriveTangent=0.3137, interpMode=RCIM_Cubic, leaveTangent=0.3137, tangentMode=RCTM_User, time=0, value=1}
         {arriveTangent=-2.0958, interpMode=RCIM_Cubic, leaveTangent=-2.0958, tangentMode=RCTM_User, time=1, value=0}
        ]
       }}})
    Scale RGB=&VectorFromCurve(VectorCurve=DataInterface() {propertyValues {
       XCurve {keys [{interpMode=RCIM_Cubic, tangentMode=RCTM_User, time=0, value=1}, {arriveTangent=-0.2294, tangentMode=RCTM_User, time=1, value=0}]}
       YCurve {keys [{interpMode=RCIM_Cubic, tangentMode=RCTM_User, time=0, value=1}, {arriveTangent=-0.2294, tangentMode=RCTM_User, time=1, value=0}]}
       ZCurve {keys [{interpMode=RCIM_Cubic, tangentMode=RCTM_User, time=0, value=1}, {arriveTangent=-0.2294, tangentMode=RCTM_User, time=1, value=0}]}
      }})
    ScaleA=true
    ScaleRGB=true
   )
   &SolveForcesAndVelocity(Clamp Velocity=false, Limit Acceleration=false)
   /SetVariables(Particles.Glow Color=&Multiply_LinearColorByFloat(
     Linear Color=&MultiplyLinearColor(
      Linear Color=$Particles.Color
      Scale Factor=&MakeVector4_FromLinearColor(LinearColor=<rgba 1 0.5877 0.8563 1>)
     )
     Float=0.33
    )) {isSetParameters=true, module=SetVariables_#7}
  ]
  rendererClasses [&NiagaraSpriteRendererProperties]
  renderers [
   {rendererClass=&NiagaraSpriteRendererProperties}
   {rendererClass=&NiagaraSpriteRendererProperties, rendererIndex=1}
  ]
 }
 SmokeRing {
  emitterUpdateScript [
   &EmitterState(Life Cycle Mode=::NewEnumerator1, Loop Behavior=::NewEnumerator1, Loop Duration=0.3, UseLoopDelay=false)
   &SpawnBurst_Instantaneous(Spawn Count=8, Use Spawn Probability=false, UseLoopCountLimit=false)
   &SpawnRate(
    SpawnRate=&FloatFromCurve(FloatCurve=DataInterface() {propertyValues {Curve {keys [<at 0 0.95>,<at 0.5 0.95>,<at 1 0>]}}}, CurveIndex=$Emitter.Age, Scale Curve=20)
    Use Spawn Probability=false
   )
  ]
  particleSpawnScript [
   &InitializeParticle(
    Color Maximum=<rgba 0.3949 0.3949 0.3949 1>
    Color Minimum=<rgba 0.5129 0.5129 0.5129 1>
    Color Mode=::NewEnumerator2
    Lifetime Max=0.6
    Lifetime Min=0.4
    Lifetime Mode=::NewEnumerator1
    Mesh Scale=&Multiply_VectorByFloat(
     Vector=&RandomRangeVector(Minimum=<1 1 1.4>, Maximum=<1 1 1.8>)
     Float=&RandomRangeFloat(Minimum=0.4, Maximum=0.5)
    )
    Mesh Scale Mode=::NewEnumerator4
    UsePositionOffset=false
    Write Material Random=true
   )
   &InitialMeshOrientation(
    Mesh Orientation Mode=::NewEnumerator2
    Rotation=&RandomRangeVector(Minimum=<0.02 0.02 -1>, Maximum=<0.03 0.03 1>)
    Use Rotation Vector=true
   )
   &ShapeLocation(
    Cylinder Height=8
    Cylinder Radius=1
    MaskSpawnGroup=false
    Offset=<0 0 -25>
    Offset Mode=::NewEnumerator0
    Override Seed=false
    Shape Primitive=::NewEnumerator1
    Surface Only=false
    UseOwnerScale=true
   )
  ]
  particleUpdateScript [
   &ParticleState()
   &DynamicMaterialParameters(
    Index 0 Param 1=&FloatFromCurve(FloatCurve=DataInterface() {propertyValues {Curve {keys [{interpMode=RCIM_Cubic, tangentMode=RCTM_User, time=0, value=0.8}, {interpMode=RCIM_Cubic, tangentMode=RCTM_User, time=1, value=0}]}}})
    Module.Param0WriteEnabled=true
    Module.Param1WriteEnabled=true
    Module.Param2WriteEnabled=true
    Module.Param3WriteEnabled=true
   )
   &ScaleMeshSize(Scale Factor=&Multiply_VectorByFloat(
     Vector=&Multiply_VectorByFloat(Vector=$Engine.Owner.Scale, Float=&FN_FrontEndFloatModifier(Gameplay=0.5, Front End=0.5))
     Float=&FloatFromCurve(FloatCurve=DataInterface() {propertyValues {Curve {keys [
          {arriveTangent=0.5864, interpMode=RCIM_Cubic, leaveTangent=0.5864, tangentMode=RCTM_User, time=0, value=0.5}
          {arriveTangent=0.4564, interpMode=RCIM_Cubic, leaveTangent=0.4564, tangentMode=RCTM_User, time=0.5954, value=0.6852}
          {arriveTangent=0.5662, interpMode=RCIM_Cubic, leaveTangent=0.5662, tangentMode=RCTM_User, time=1, value=1.1}
         ]}}})
    ))
   &ScaleMeshSize(Scale Factor=&MakeVector(
     X=&FloatFromCurve(
      FloatCurve=DataInterface() {propertyValues {Curve {keys [
          {arriveTangent=0.3413, interpMode=RCIM_Cubic, leaveTangent=0.3413, tangentMode=RCTM_User, time=0.8, value=1}
          {arriveTangent=3.5208, interpMode=RCIM_Cubic, leaveTangent=3.5208, tangentMode=RCTM_User, time=1, value=1.5}
         ]}}}
      CurveIndex=$Emitter.Age
     )
     Y=&FloatFromCurve(
      FloatCurve=DataInterface() {propertyValues {Curve {keys [
          {arriveTangent=0.3598, interpMode=RCIM_Cubic, leaveTangent=0.3598, tangentMode=RCTM_User, time=0.8, value=1}
          {arriveTangent=3.5574, interpMode=RCIM_Cubic, leaveTangent=3.5574, tangentMode=RCTM_User, time=1, value=1.5}
         ]}}}
      CurveIndex=$Emitter.Age
     )
     Z=1
    )) {enabled=false, module=ScaleMeshSize001}
   &UpdateMeshOrientation(
    Rotation Rate=&Multiply_Float(
     A=-0.8
     B=&FloatFromCurve(FloatCurve=DataInterface() {propertyValues {Curve {keys [<at 0 1>,<at 0.4 1>,<at 1 0>]}}})
    )
    Rotation Vector=<0 0 1>
   )
   &ScaleColor(
    Scale Alpha=&FloatFromCurve(
     FloatCurve=DataInterface() {propertyValues {Curve {keys [
         {arriveTangent=4.9076, interpMode=RCIM_Cubic, leaveTangent=4.9076, tangentMode=RCTM_User, time=0, value=0}
         {arriveTangent=0.0332, interpMode=RCIM_Cubic, leaveTangent=0.0332, tangentMode=RCTM_User, time=0.1, value=1}
         {arriveTangent=-0.0163, interpMode=RCIM_Cubic, leaveTangent=-0.025, tangentMode=RCTM_Break, time=0.4, value=1}
         {arriveTangent=-1.5, interpMode=RCIM_Cubic, leaveTangent=-1.5, time=0.6, value=0.8}
         {arriveTangent=-2, interpMode=RCIM_Cubic, leaveTangent=-2, time=0.9, value=0.25}
         {arriveTangent=-3.0089, tangentMode=RCTM_User, time=1, value=0}
        ]}}}
     Scale Curve=2
    )
    Scale RGB=&MakeVectorFromLinearColorRGB(Color=&ColorFromCurve(ColorCurve=DataInterface() {propertyValues {
        AlphaCurve {keys []}
        BlueCurve {keys [{arriveTangent=-0.3927, interpMode=RCIM_Cubic, time=0.4397, value=1.1047}, <at 0.6621 0.777 RCIM_Cubic>]}
        GreenCurve {keys [{arriveTangent=-0.0196, interpMode=RCIM_Cubic, time=0.4397, value=0.3512}, <at 0.6621 0 RCIM_Cubic>]}
        RedCurve {keys [{arriveTangent=-0.2024, interpMode=RCIM_Cubic, time=0.4397, value=1.1585}, <at 0.6621 0.2331 RCIM_Cubic>]}
       }}))
    ScaleA=true
    ScaleRGB=true
   )
   &SolveForcesAndVelocity(Clamp Velocity=false, Limit Acceleration=false)
   /SetVariables(Particles.Position=&FN_FrontEndVector3Modifier(
     Gameplay=&Lerp_Vector(
      A=$System.HeartIntroSpawnPosition
      B=&RandomRangeVector(Minimum=<0 0 0>, Maximum=<0 0 25>)
      Alpha=$Particles.Age
     )
     Front End=&Lerp_Vector(A=<0 0 -20>, B=&RandomRangeVector(Minimum=<0 0 0>, Maximum=<0 0 10>), Alpha=$Particles.Age)
    )) {isSetParameters=true}
  ]
  rendererClasses [&NiagaraMeshRendererProperties]
  renderers [{rendererClass=&NiagaraMeshRendererProperties, rendererIndex=0}]
 }
}
summary {
 systemName=NS_ReactFX_BarrelWimp
 emitters [
  {emitterName=Intro, rendererClasses [&NiagaraSpriteRendererProperties]}
  {emitterName=Pop, rendererClasses [&NiagaraSpriteRendererProperties]}
  {emitterName=InkSmoke, rendererClasses [&NiagaraSpriteRendererProperties]}
  {emitterName=SmokeRing, rendererClasses [&NiagaraMeshRendererProperties]}
  {emitterName=Lightning001, rendererClasses [&NiagaraMeshRendererProperties]}
  {emitterName=Glow_n_Pop, rendererClasses [&NiagaraSpriteRendererProperties]}
 ]
 userVariables [
  User.bIsFrontEnd {type=NiagaraBool}
  User.bIsFrontEndPreview {type=NiagaraBool}
  User.FrontEndScale {type=NiagaraFloat, defaultValue=0.7, description=""}
 ]
}
