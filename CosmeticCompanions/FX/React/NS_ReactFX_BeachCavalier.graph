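# Niagara reaction effect for the BeachCavalier cosmetic companion, pairing a sprite pop with sand puffs and the shared petting-reaction sound.
# Intro plays the audio and bursts sub-UV sprites with glow parameters, timed and offset by system-level values.
# Sand_Intro and delayed Sand_Outro place streaks in five partitions, animating height, stretch and fade through curves instead of forces.
# Time-of-day functions tint sprites, front-end modifiers lower the sand, and Minimal is disabled.
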NiagaraSystem /CosmeticCompanions/FX/React/NS_ReactFX_BeachCavalier
dependencies {
 usedDataInterfaces [
  &NiagaraDataInterfaceAudioPlayer
  &NiagaraDataInterfaceCurve
  &NiagaraDataInterfaceSpriteRendererInfo
  &NiagaraDataInterfaceMeshRendererInfo
  &NiagaraDataInterfaceVector2DCurve
  &NiagaraDataInterfaceColorCurve
 ]
 usedDynamicInputs []
 usedModules [
  &self:SystemSpawnScript.SystemScriptSource.SystemScriptGraph.NiagaraNodeAssignment_0.SetVariables_#1
  &self:SystemSpawnScript.SystemScriptSource.SystemScriptGraph.NiagaraNodeAssignment_1.SetVariables_#2
  &self:SystemSpawnScript.SystemScriptSource.SystemScriptGraph.NiagaraNodeAssignment_4.SetVariables_#3
  &self:SystemSpawnScript.SystemScriptSource.SystemScriptGraph.NiagaraNodeAssignment_2.SetVariables_#4
  &self:SystemSpawnScript.SystemScriptSource.SystemScriptGraph.NiagaraNodeAssignment_3.SetVariables_#5
  &self:SystemSpawnScript.SystemScriptSource.SystemScriptGraph.NiagaraNodeAssignment_9.SetVariables_#6
  &self:SystemSpawnScript.SystemScriptSource.SystemScriptGraph.NiagaraNodeAssignment_10.SetVariables_#7
  &self:SystemSpawnScript.SystemScriptSource.SystemScriptGraph.NiagaraNodeAssignment_5.SetVariables_#8
  &self:SystemSpawnScript.SystemScriptSource.SystemScriptGraph.NiagaraNodeAssignment_6.SetVariables_#9
  &self:SystemSpawnScript.SystemScriptSource.SystemScriptGraph.NiagaraNodeAssignment_7.SetVariables_#10
  &SystemState
  &self:SystemSpawnScript.SystemScriptSource.SystemScriptGraph.NiagaraNodeAssignment_8.SetVariables_#11
 ]
}
emitters {
 Ambient_Sand {rendererClasses [], renderers []}
 Ambient_Sand001 {rendererClasses [], renderers []}
 Intro {
  emitterSpawnScript [
   &InitializeAudioPlayer(Audio Player=DataInterface() {propertyValues {
      Attenuation=null
      Concurrency=null
      MaxPlaysPerTick=10
      SoundToPlay=/CosmeticCompanions/Assets/Audio_Shared/Interaction/Petting/SC_Mimosa_Interaction_React_Default
      bAllowLoopingOneShotSounds=false
      bLimitPlaysPerTick=true
      bOnlyActiveDuringGameplay=false
      bStopWhenComponentIsDestroyed=true
      ConfigurationUserParameter {parameter {name=None, typeDefHandle {registeredTypeIndex=1}, varData []}}
      ParameterNames []
     }})
   /Niagara/Modules/Audio/PlayAudio() {module=PlayAudio001}
  ]
  emitterUpdateScript [
   &EmitterState(
    Inactive Response=::NewEnumerator1
    Life Cycle Mode=::NewEnumerator1
    Loop Behavior=::NewEnumerator1
    Loop Delay=0.12
    Loop Duration=1.3
    UseLoopDelay=true
   )
   &SpawnBurst_Instantaneous(Spawn Time=0.16, Use Spawn Probability=false, UseLoopCountLimit=false)
   &SpawnBurst_Instantaneous(Spawn Group=1, Spawn Time=0.16, Use Spawn Probability=false, UseLoopCountLimit=false) {module=SpawnBurst_Instantaneous001}
  ]
  particleSpawnScript [
   /SetVariables(Particles.VisibilityTag=&SpawnGroup()) {isSetParameters=true}
   &InitializeParticle(
    Color=&FNTimeOfDayColor(
     Night Color=<rgba 0.4314 0.5862 0.8073 1>
     Morning Color=<rgba 0.7055 0.6772 0.824 1>
     Day Color=<rgba 0.6955 0.7947 0.9365 1>
     Evening Color=<rgba 0.6472 0.6447 0.824 1>
    )
    Lifetime=$System.IntroLength
    Sprite Rotation Mode=::NewEnumerator0
    Sprite Size Mode=::NewEnumerator3
    Uniform Sprite Size=20
    UsePositionOffset=false
    Write Material Random=true
   )
  ]
  particleUpdateScript [
   &ParticleState()
   &ScaleSpriteSize(
    Scale Sprite Size Mode=::NewEnumerator0
    Uniform Scale Factor=&Add_Float(
     A=0.5
     B=&FloatFromCurve(FloatCurve=DataInterface() {propertyValues {Curve {
         defaultValue=0
         keys [
          {arriveTangent=23.9023, interpMode=RCIM_Cubic, leaveTangent=23.9023, tangentMode=RCTM_User, time=0, value=0}
          {arriveTangent=0.9194, interpMode=RCIM_Cubic, leaveTangent=3.5477, tangentMode=RCTM_Break, time=0.2438, value=0.2479}
          {arriveTangent=1.1355, interpMode=RCIM_Cubic, leaveTangent=1.1355, tangentMode=RCTM_User, time=0.4721, value=0.7957}
          {
           arriveTangent=0.7231
           arriveTangentWeight=0.1198
           interpMode=RCIM_Cubic
           leaveTangent=0.7295
           tangentMode=RCTM_Break
           time=0.894
           value=0.9527
          }
          {arriveTangent=-16.5925, arriveTangentWeight=0.0561, interpMode=RCIM_Cubic, leaveTangent=-16.5925, tangentMode=RCTM_User, time=1, value=1}
         ]
        }}})
    )
   ) {enabled=false, module=ScaleSpriteSize002}
   &ScaleColor(
    Scale Alpha=&FloatFromCurve(FloatCurve=DataInterface() {propertyValues {Curve {
        defaultValue=0
        keys [
         {arriveTangent=29.0642, interpMode=RCIM_Cubic, leaveTangent=29.0642, tangentMode=RCTM_User, time=0, value=0}
         {arriveTangent=-0.2441, leaveTangent=-0.2441, time=0.07, value=0.9847}
         {arriveTangent=-1.146, leaveTangent=-1.146, time=1, value=1}
         {arriveTangent=-0.0405, interpMode=RCIM_Cubic, leaveTangent=-0.0405, tangentMode=RCTM_User, time=1, value=0}
        ]
       }}})
    Scale RGB=&VectorFromFloat(Value=1)
    ScaleA=true
    ScaleRGB=true
   )
   &OffsetPosition(Position Offset=$System.BurstPosition)
   &DynamicMaterialParameters(
    Index 0 Param 1=&FloatFromCurve(FloatCurve=DataInterface() {propertyValues {Curve {keys [{arriveTangent=0.6296, interpMode=RCIM_Cubic, leaveTangent=0.6296, tangentMode=RCTM_User, time=0, value=0}, <at 0.75 0.9676>, <at 1 1>]}}})
    Index 0 Param 2=0.2
    Module.Index0Float0Write=true
    Module.Index0Float1Write=true
    Module.Index0Float2Write=true
    Module.Index0Float3Write=true
   )
   /SetVariables(
    Particles.Glow Color=&MakeLinearColorFromVectorAndFloat(
     "Vector (RGB)"=&MakeVectorFromLinearColorRGB(Color=<rgba 2 1.3066 0.75 1>)
     "Float (Alpha)"=&FloatFromCurve(
      FloatCurve=DataInterface() {propertyValues {Curve {
         defaultValue=1
         keys [
          <at 0 0 RCIM_Cubic>
          {arriveTangent=1.4286, interpMode=RCIM_Cubic, leaveTangent=1.4286, time=0.4, value=1}
          {arriveTangent=-0.0527, interpMode=RCIM_Cubic, leaveTangent=-0.0527, tangentMode=RCTM_User, time=0.7, value=1}
          <at 0.9 0 RCIM_Cubic>
         ]
        }}}
      Scale Curve=0.125
     )
    )
    Particles.Glow Size=&Multiply_Vector2D(A=$Particles.SpriteSize, B=&Vector2DFromFloat(Value=1.6267))
    Particles.GlowDP:Vector4f=<0 0 0 0>
   ) {isSetParameters=true}
   &SubUVAnimation(
    End Frame Range Override=8
    Output Alternate Renderer Binding=true
    "Sprite Renderer "=DataInterface() {propertyValues {SpriteRenderer=&self:Intro_0.NiagaraSpriteRendererProperties_2}}
    SubUV Animation Mode:ENiagara_SubUVLookupModeV2=::NewEnumerator1
    SubUV Curve=DataInterface() {propertyValues {Curve {
       defaultValue=0
       keys [
        {arriveTangent=2.5185, interpMode=RCIM_Cubic, leaveTangent=2.5185, tangentMode=RCTM_User, time=0, value=0}
        {leaveTangent=2.9618, time=0.225, value=0.8829}
        {arriveTangent=-0.5448, interpMode=RCIM_Cubic, leaveTangent=-0.5448, tangentMode=RCTM_User, time=0.7, value=1}
        {arriveTangent=-0.0804, interpMode=RCIM_Cubic, leaveTangent=-0.0804, tangentMode=RCTM_User, time=0.925, value=0}
       ]
      }}}
    UseEndFrame=true
    UseStartFrame=false
   )
  ]
  rendererClasses [&NiagaraSpriteRendererProperties]
  renderers [
   {rendererClass=&NiagaraSpriteRendererProperties}
   {rendererClass=&NiagaraSpriteRendererProperties, rendererIndex=1}
  ]
 }
 Minimal {
  bEnabled=false
  emitterUpdateScript [
   &EmitterState()
   &SpawnBurst_Instantaneous(Spawn Count=16, Use Spawn Probability=false, UseLoopCountLimit=false)
   &SpawnRate(
    SpawnRate=&FloatFromCurve(
     FloatCurve=DataInterface() {propertyValues {Curve {keys [<at 0 1>,<at 0.075 1>,<at 0.125 0>]}}}
     CurveIndex=$Emitter.NormalizedLoopAge
     Scale Curve=52
    )
    Use Spawn Probability=false
   )
  ]
  particleSpawnScript [
   &PartitionParticles(Number Of Partitions=5)
   &InitializeParticle(
    Color Maximum=<rgba 0.3515 0.2582 0.0723 0.8>
    Color Minimum=<rgba 0.3813 0.2874 0.1095 0.6>
    Color Mode=::NewEnumerator2
    Lifetime Max=1.2
    Lifetime Min=0.7
    Lifetime Mode=::NewEnumerator1
    Position Offset=&MergeHelper_VectorSwitch(
     0=<0 0 31.3711>
     1=<4 0 15.68>
     2=<-4 0 21.4006>
     3=<0 4 15.68>
     4=<0 -4 21.4006>
     Index=$Particles.PartitionParticles.Partition
     Num=::NewEnumerator3
    )
    Sprite Rotation Mode=::NewEnumerator2
    Sprite Size Mode=::NewEnumerator1
    Uniform Sprite Size Max=9
    UsePositionOffset=true
    Write Material Random=true
   )
   &CameraOffset(Camera Offset Amount=25)
   &AddVelocity(
    Cone Axis=<0 0 -1>
    Module.UseConeFalloff=false
    Module.UseVelDistribution=false
    Override Seed=false
    Velocity Mode=::NewEnumerator2
    Velocity Speed=&RandomRangeFloat(Minimum=5, Maximum=8)
   )
  ]
  particleUpdateScript [
   &ParticleState()
   &SpriteRotationRate()
   &ScaleSpriteSize(Uniform Curve Sprite Scale=DataInterface() {propertyValues {Curve {
       defaultValue=0
       keys [
        {arriveTangent=4.1785, interpMode=RCIM_Cubic, leaveTangent=4.1785, tangentMode=RCTM_User, time=0, value=0}
        {arriveTangent=1.1179, interpMode=RCIM_Cubic, leaveTangent=1.1179, tangentMode=RCTM_User, time=0.2042, value=0.6228}
        {arriveTangent=0.0224, interpMode=RCIM_Cubic, leaveTangent=0.0224, tangentMode=RCTM_User, time=1, value=1}
       ]
      }}})
   &SubUVAnimation(
    "Sprite Renderer "=DataInterface() {propertyValues {SpriteRenderer=&self:Minimal.NiagaraSpriteRendererProperties_1}}
    UseEndFrame=false
    UseStartFrame=false
   )
   &ScaleColor(
    Scale Alpha=&FloatFromCurve(
     FloatCurve=DataInterface() {propertyValues {Curve {
        defaultValue=1
        keys [
         {arriveTangent=9.0725, interpMode=RCIM_Cubic, leaveTangent=9.0725, tangentMode=RCTM_User, time=0, value=0}
         {arriveTangent=-2.5249, interpMode=RCIM_Cubic, leaveTangent=-2.5249, tangentMode=RCTM_User, time=0.1408, value=0.9847}
         {arriveTangent=-0.0405, interpMode=RCIM_Cubic, leaveTangent=-0.0405, tangentMode=RCTM_User, time=1, value=0}
        ]
       }}}
     Scale Curve=&RandomRangeFloat(Minimum=0.3, Maximum=0.7)
    )
    ScaleA=true
    ScaleRGB=true
   )
   &SolveForcesAndVelocity(Clamp Velocity=false, Limit Acceleration=false)
   &DynamicMaterialParameters(
    Index 0 Param 1=1
    Index 0 Param 2=1
    Module.Index0Float0Write=true
    Module.Index0Float1Write=true
    Module.Index0Float2Write=true
    Module.Index0Float3Write=true
   )
  ]
  rendererClasses [&NiagaraSpriteRendererProperties]
  renderers [{rendererClass=&NiagaraSpriteRendererProperties}]
 }
 Sand_Intro {
  emitterUpdateScript [
   &EmitterState()
   &SpawnBurst_Instantaneous(Spawn Count=4, Use Spawn Probability=false, UseLoopCountLimit=false)
   &SpawnRate(
    SpawnRate=&FloatFromCurve(
     FloatCurve=DataInterface() {propertyValues {Curve {keys [<at 0 1>,<at 0.125 1>,<at 0.2 0>]}}}
     CurveIndex=$Emitter.NormalizedLoopAge
     Scale Curve=32
    )
    Use Spawn Probability=false
   )
   /SetVariables(Emitter.DepthFade=40) {isSetParameters=true, module=SetVariables_#12}
  ]
  particleSpawnScript [
   &PartitionParticles(Number Of Partitions=5)
   &InitializeParticle(
    Color Maximum=<rgba 0.651 0.3842 0.1173 1>
    Color Minimum=<rgba 0.5149 0.4233 0.1946 1>
    Color Mode=::NewEnumerator2
    Lifetime Max=0.8
    Lifetime Min=0.65
    Lifetime Mode=::NewEnumerator1
    Sprite Rotation Mode=::NewEnumerator2
    Sprite Size=&Multiply_Vector2D_ByFloat(
     Float=&MergeHelper_FloatSwitch(0=1, 1=0.4, 2=0.7, 3=0.4, 4=0.7, Index=$Particles.PartitionParticles.Partition, Num=::NewEnumerator3)
     Vector2D=&RandomRangeVector2D(Minimum=<12.5 25>, Maximum=<12.5 37.5>)
    )
    Sprite Size Mode=::NewEnumerator4
    UsePositionOffset=false
    Write Material Random=true
   )
   &SubUVAnimation(
    End Frame Range Override=3
    Evaluation Type=::NewEnumerator0
    "Sprite Renderer "=DataInterface() {propertyValues {SpriteRenderer=&self:Sand_Intro_0.NiagaraSpriteRendererProperties_1}}
    SubUV Animation Mode:ENiagara_SubUVLookupModeV2=::NewEnumerator2
    UseEndFrame=false
    UseStartFrame=false
   )
  ]
  particleUpdateScript [
   &ParticleState()
   &GravityForce(Gravity=&MakeVector(
     X=0
     Y=0
     Z=&MergeHelper_FloatSwitch(0=-156.6119, 1=-152.7235, 2=-54.5526, 3=?, Index=$Particles.PartitionParticles.Partition, Num=::NewEnumerator1)
    )) {enabled=false}
   &SolveForcesAndVelocity(Clamp Velocity=false, Limit Acceleration=false)
   &ScaleSpriteSize(
    Scale Factor=&MakeVector2D(
     X=&FloatFromCurve(FloatCurve=DataInterface() {propertyValues {Curve {keys [
          {arriveTangent=1.3188, interpMode=RCIM_Cubic, time=0, value=0.2}
          {arriveTangent=0.7857, interpMode=RCIM_Cubic, leaveTangent=0.7857, tangentMode=RCTM_User, time=1, value=0.5}
         ]}}})
     Y=&FloatFromCurve(FloatCurve=DataInterface() {propertyValues {Curve {keys [
          {interpMode=RCIM_Cubic, tangentMode=RCTM_User, time=0, value=1.2}
          {arriveTangent=-2.3605, interpMode=RCIM_Cubic, leaveTangent=-2.3605, tangentMode=RCTM_User, time=1, value=0}
         ]}}})
    )
    Scale Sprite Size Mode=::NewEnumerator2
   )
   &ScaleColor(
    Scale Alpha=&FloatFromCurve(FloatCurve=DataInterface() {propertyValues {Curve {keys [
         {interpMode=RCIM_Cubic, leaveTangentWeight=0.0962, tangentMode=RCTM_Break, time=0.05, value=0}
         {interpMode=RCIM_Cubic, tangentMode=RCTM_Break, time=0.2, value=1}
         {arriveTangent=-1.8058, interpMode=RCIM_Cubic, tangentMode=RCTM_Break, time=1, value=0}
        ]}}})
    Scale RGB=&Multiply_VectorByFloat(
     Vector=&MakeVectorFromLinearColorRGB(Color=&MultiplyLinearColor(
       Linear Color=&ColorFromCurve(ColorCurve=DataInterface() {propertyValues {
          AlphaCurve {keys [<at 0 1 RCIM_Cubic>]}
          BlueCurve {keys [
            <at 0.0061 0.1042>
            {arriveTangent=1.7407, interpMode=RCIM_Cubic, leaveTangent=1.7407, time=0.1571, value=0.4323}
            <at 0.3801 0.7552 RCIM_Cubic>
           ]}
          GreenCurve {keys [
            <at 0.0061 0.1042>
            {arriveTangent=1.7407, interpMode=RCIM_Cubic, leaveTangent=1.7407, time=0.1571, value=0.4323}
            <at 0.3801 0.7552 RCIM_Cubic>
           ]}
          RedCurve {keys [
            <at 0.0061 0.1042>
            {arriveTangent=1.7407, interpMode=RCIM_Cubic, leaveTangent=1.7407, time=0.1571, value=0.4323}
            <at 0.3801 0.7552 RCIM_Cubic>
           ]}
         }})
       Scale Factor=&MakeVector4_FromLinearColor(LinearColor=&FNTimeOfDayColor(
         Night Color=<rgba 0.3066 0.4742 0.8177 1>
         Morning Color=<rgba 1 0.3015 0.2916 1>
         Day Color=<rgba 1 1 1 1>
         Evening Color=<rgba 1 0.5153 0.1302 1>
        ))
      ))
     Float=1
    )
    ScaleA=true
    ScaleRGB=true
   )
   &CameraOffset(Camera Offset Amount=&FloatFromCurve(
     FloatCurve=DataInterface() {propertyValues {Curve {keys [
         {arriveTangent=3.1425, interpMode=RCIM_Cubic, leaveTangent=3.1425, tangentMode=RCTM_User, time=0.35, value=0}
         {interpMode=RCIM_Cubic, leaveTangent=1.9808, time=1, value=1}
        ]}}}
     Scale Curve=20
    ))
   &DynamicMaterialParameters(
    Index 0 Param 1=&RandomRangeFloat(Minimum=0, Maximum=1)
    Index 0 Param 2=1
    Index 0 Param 3=3
    Index 0 Param 4=&FloatFromCurve(FloatCurve=DataInterface() {propertyValues {Curve {keys [
         <at 0 0 RCIM_Cubic>
         {arriveTangent=2, leaveTangent=2, time=0.2, value=1}
         {interpMode=RCIM_Cubic, tangentMode=RCTM_Break, time=0.5, value=1}
         {arriveTangent=-2.791, interpMode=RCIM_Cubic, tangentMode=RCTM_Break, time=1, value=0.25}
        ]}}})
    Module.Index0Param0Write=true
    Module.Index0Param1Write=true
    Module.Index0Param2Write=true
    Module.Index0Param3Write=true
   )
   /SetVariables(Particles.Position=&AddVectorToPosition(
     Position=&AddVectorToPosition(
      Position=$Engine.Emitter.SimulationPosition
      Vector=&AddVectorToPosition(
       Position=&MergeHelper_VectorSwitch(
        0=<0 0 10.8629>
        1=<4 0 7.3057>
        2=<-4 0 4.525>
        3=<0 4 7.3057>
        4=<0 -4 4.525>
        Index=$Particles.PartitionParticles.Partition
        Num=::NewEnumerator3
       )
       Vector=&MakeVector(
        X=0
        Y=0
        Z=&FloatFromCurve(
         FloatCurve=DataInterface() {propertyValues {Curve {keys [{arriveTangent=2.4514, interpMode=RCIM_Cubic, leaveTangent=2.4514, tangentMode=RCTM_User, time=0, value=0}, <at 1 1 RCIM_Cubic>]}}}
         Scale Curve=&MergeHelper_FloatSwitch(0=-23.0492, 1=-13.2461, 2=-8.3789, 3=-13.2461, 4=-8.3789, Index=$Particles.PartitionParticles.Partition, Num=::NewEnumerator3)
        )
       )
      )
     )
     Vector=&FN_FrontEndVector3Modifier(Gameplay=<0 0 0>, Front End=<0 0 -15>)
    )) {isSetParameters=true, module=SetVariables_#13}
  ]
  rendererClasses [&NiagaraSpriteRendererProperties]
  renderers [{rendererClass=&NiagaraSpriteRendererProperties}]
 }
 Sand_Outro {
  emitterUpdateScript [
   &EmitterState(Life Cycle Mode=::NewEnumerator1, Loop Behavior=::NewEnumerator1, Loop Delay=0.72, Loop Duration=0.5, UseLoopDelay=true)
   &SpawnBurst_Instantaneous(Spawn Count=4, Use Spawn Probability=false, UseLoopCountLimit=false)
   &SpawnBurst_Instantaneous(Spawn Count=4, Spawn Time=0.08, Use Spawn Probability=false, UseLoopCountLimit=false) {module=SpawnBurst_Instantaneous001}
   &SpawnBurst_Instantaneous(Spawn Count=4, Spawn Time=0.16, Use Spawn Probability=false, UseLoopCountLimit=false) {module=SpawnBurst_Instantaneous002}
   &SpawnRate(SpawnRate=32, Use Spawn Probability=false) {enabled=false}
   /SetVariables(Emitter.DepthFade=40) {isSetParameters=true, module=SetVariables_#12}
  ]
  particleSpawnScript [
   &PartitionParticles(Number Of Partitions=5)
   &InitializeParticle(
    Color Maximum=<rgba 0.651 0.3842 0.1173 1>
    Color Minimum=<rgba 0.5149 0.4233 0.1946 1>
    Color Mode=::NewEnumerator2
    Lifetime Max=0.8
    Lifetime Min=0.65
    Lifetime Mode=::NewEnumerator1
    Sprite Rotation Mode=::NewEnumerator2
    Sprite Size=&Multiply_Vector2D_ByFloat(
     Float=&MergeHelper_FloatSwitch(0=1, 1=0.7, 2=1, 3=0.7, 4=1, Index=$Particles.PartitionParticles.Partition, Num=::NewEnumerator3)
     Vector2D=&RandomRangeVector2D(Minimum=<12.5 25>, Maximum=<12.5 37.5>)
    )
    Sprite Size Mode=::NewEnumerator4
    UsePositionOffset=false
    Write Material Random=true
   )
   &SubUVAnimation(
    End Frame Range Override=3
    Evaluation Type=::NewEnumerator0
    "Sprite Renderer "=DataInterface() {propertyValues {SpriteRenderer=&self:Sand_Outro_0.NiagaraSpriteRendererProperties_1}}
    SubUV Animation Mode:ENiagara_SubUVLookupModeV2=::NewEnumerator2
    UseEndFrame=false
    UseStartFrame=false
   )
  ]
  particleUpdateScript [
   &ParticleState()
   &GravityForce(Gravity=&MakeVector(
     X=0
     Y=0
     Z=&MergeHelper_FloatSwitch(0=-156.6119, 1=-152.7235, 2=-54.5526, 3=?, Index=$Particles.PartitionParticles.Partition, Num=::NewEnumerator1)
    )) {enabled=false}
   &SolveForcesAndVelocity(Clamp Velocity=false, Limit Acceleration=false)
   &ScaleSpriteSize(
    Scale Factor=&MakeVector2D(
     X=&FloatFromCurve(FloatCurve=DataInterface() {propertyValues {Curve {keys [
          {arriveTangent=1.3188, interpMode=RCIM_Cubic, time=0, value=0.2}
          {arriveTangent=0.7857, interpMode=RCIM_Cubic, leaveTangent=0.7857, tangentMode=RCTM_User, time=1, value=0.5}
         ]}}})
     Y=&FloatFromCurve(FloatCurve=DataInterface() {propertyValues {Curve {keys [
          {interpMode=RCIM_Cubic, tangentMode=RCTM_User, time=0, value=1.2}
          {arriveTangent=-2.3605, interpMode=RCIM_Cubic, leaveTangent=-2.3605, tangentMode=RCTM_User, time=1, value=0}
         ]}}})
    )
    Scale Sprite Size Mode=::NewEnumerator2
   )
   &ScaleColor(
    Scale Alpha=&FloatFromCurve(
     FloatCurve=DataInterface() {propertyValues {Curve {keys [
         {interpMode=RCIM_Cubic, leaveTangentWeight=0.0962, tangentMode=RCTM_Break, time=0.05, value=0}
         {interpMode=RCIM_Cubic, tangentMode=RCTM_Break, time=0.2, value=1}
         {arriveTangent=-1.8058, interpMode=RCIM_Cubic, tangentMode=RCTM_Break, time=1, value=0}
        ]}}}
     Scale Curve=&RandomRangeFloat(Minimum=0.4, Maximum=0.7)
    )
    Scale RGB=&MakeVectorFromLinearColorRGB(Color=&MultiplyLinearColor(
      Linear Color=&ColorFromCurve(ColorCurve=DataInterface() {propertyValues {
         AlphaCurve {keys [<at 0 1 RCIM_Cubic>]}
         BlueCurve {keys [
           <at 0.0061 0.1042>
           {arriveTangent=1.7407, interpMode=RCIM_Cubic, leaveTangent=1.7407, time=0.1571, value=0.4323}
           <at 0.3801 0.7552 RCIM_Cubic>
          ]}
         GreenCurve {keys [
           <at 0.0061 0.1042>
           {arriveTangent=1.7407, interpMode=RCIM_Cubic, leaveTangent=1.7407, time=0.1571, value=0.4323}
           <at 0.3801 0.7552 RCIM_Cubic>
          ]}
         RedCurve {keys [
           <at 0.0061 0.1042>
           {arriveTangent=1.7407, interpMode=RCIM_Cubic, leaveTangent=1.7407, time=0.1571, value=0.4323}
           <at 0.3801 0.7552 RCIM_Cubic>
          ]}
        }})
      Scale Factor=&MakeVector4_FromLinearColor(LinearColor=&FNTimeOfDayColor(
        Night Color=<rgba 0.3066 0.4742 0.8177 1>
        Morning Color=<rgba 1 0.3015 0.2916 1>
        Day Color=<rgba 1 1 1 1>
        Evening Color=<rgba 1 0.5153 0.1302 1>
       ))
     ))
    ScaleA=true
    ScaleRGB=true
   )
   &CameraOffset(Camera Offset Amount=&FloatFromCurve(
     FloatCurve=DataInterface() {propertyValues {Curve {keys [
         {arriveTangent=3.1425, interpMode=RCIM_Cubic, leaveTangent=3.1425, tangentMode=RCTM_User, time=0.35, value=0}
         {interpMode=RCIM_Cubic, leaveTangent=1.9808, time=1, value=1}
        ]}}}
     Scale Curve=20
    ))
   &DynamicMaterialParameters(
    Index 0 Param 1=&RandomRangeFloat(Minimum=0, Maximum=1)
    Index 0 Param 2=1
    Index 0 Param 3=3
    Index 0 Param 4=&FloatFromCurve(FloatCurve=DataInterface() {propertyValues {Curve {keys [<at 0 1 RCIM_Cubic>, {arriveTangent=-1.6019, interpMode=RCIM_Cubic, tangentMode=RCTM_Break, time=1, value=0}]}}})
    Module.Index0Param0Write=true
    Module.Index0Param1Write=true
    Module.Index0Param2Write=true
    Module.Index0Param3Write=true
   )
   /SetVariables(Particles.Position=&AddVectorToPosition(
     Position=&AddVectorToPosition(
      Position=$Engine.Emitter.SimulationPosition
      Vector=&AddVectorToPosition(
       Position=&MergeHelper_VectorSwitch(0=<0 0 -23.4966>, 1=<4 0 -14>, 2=<-4 0 -18>, 3=<0 4 -14>, 4=<0 -4 -18>, Index=$Particles.PartitionParticles.Partition, Num=::NewEnumerator3)
       Vector=&MakeVector(
        X=0
        Y=0
        Z=&Add_Float(
         A=0
         B=&FloatFromCurve(
          FloatCurve=DataInterface() {propertyValues {Curve {keys [{arriveTangent=2.3375, interpMode=RCIM_Cubic, leaveTangent=2.3375, tangentMode=RCTM_User, time=0, value=0}, <at 1 1 RCIM_Cubic>]}}}
          Scale Curve=&MergeHelper_FloatSwitch(0=-8.0207, 1=-4.5691, 2=-6.8711, 3=-4.5691, 4=-6.8711, Index=$Particles.PartitionParticles.Partition, Num=::NewEnumerator3)
         )
        )
       )
      )
     )
     Vector=&FN_FrontEndVector3Modifier(Gameplay=<0 0 0>, Front End=<0 0 -15>)
    )) {isSetParameters=true, module=SetVariables_#13}
  ]
  rendererClasses [&NiagaraSpriteRendererProperties]
  renderers [{rendererClass=&NiagaraSpriteRendererProperties}]
 }
}
summary {
 systemName=NS_ReactFX_BeachCavalier
 emitters [
  {emitterName=Intro, rendererClasses [&NiagaraSpriteRendererProperties]}
  {emitterName=Ambient_Sand, rendererClasses []}
  {emitterName=Sand_Intro, rendererClasses [&NiagaraSpriteRendererProperties]}
  {emitterName=Sand_Outro, rendererClasses [&NiagaraSpriteRendererProperties]}
  {bEnabled=false, emitterName=Minimal, rendererClasses [&NiagaraSpriteRendererProperties]}
  {emitterName=Ambient_Sand001, rendererClasses []}
 ]
 userVariables [
  User.bIsFrontEnd {type=NiagaraBool}
  User.bIsFrontEndPreview {type=NiagaraBool}
 ]
}
