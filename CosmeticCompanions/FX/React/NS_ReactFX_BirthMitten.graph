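# Niagara reaction effect for a cosmetic companion: a heart travels to a burst point and shatters electrically, with a petting reaction sound.
# Intro_Heart plays the sound and moves over the intro length, with a bolt flash shortly before impact.
# Timed bursts follow: layered halo/electric/shatter sprites, falling glass shards, lightning bolts and GPU sparks.
# Shared System variables drive timing, position and front-end size and dimming; the light emitter is disabled.
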NiagaraSystem /CosmeticCompanions/FX/React/NS_ReactFX_BirthMitten
dependencies {
 usedDataInterfaces [
  &NiagaraDataInterfaceAudioPlayer
  &NiagaraDataInterfaceCurve
  &NiagaraDataInterfaceSpriteRendererInfo
  &NiagaraDataInterfaceColorCurve
  &NiagaraDataInterfaceArrayColor
 ]
 usedDynamicInputs [
  &FN_FrontEndVector3Modifier
  &FN_FrontEndFloatModifier
 ]
 usedModules [
  &self:SystemSpawnScript.SystemScriptSource.SystemScriptGraph.NiagaraNodeAssignment_0.SetVariables_#1
  &self:SystemSpawnScript.SystemScriptSource.SystemScriptGraph.NiagaraNodeAssignment_7.SetVariables_#2
  &self:SystemSpawnScript.SystemScriptSource.SystemScriptGraph.NiagaraNodeAssignment_4.SetVariables_#3
  &self:SystemSpawnScript.SystemScriptSource.SystemScriptGraph.NiagaraNodeAssignment_2.SetVariables_#4
  &SystemState
 ]
}
emitters {
 Burst_Shatter_Halo001 {
  emitterUpdateScript [
   &EmitterState(Life Cycle Mode=::NewEnumerator1, Loop Behavior=::NewEnumerator1, Loop Duration=2, UseLoopDelay=false)
   &SpawnBurst_Instantaneous(Spawn Time=&Add_Float(A=$System.IntroLength, B=-0.04), Use Spawn Probability=false, UseLoopCountLimit=false)
  ]
  particleSpawnScript [&InitializeParticle(
    Color Mode=::NewEnumerator0
    Lifetime=0.25
    Position:Vector3f=$System.BurstPosition
    Position Mode=::NewEnumerator1
    Sprite Size Mode=::NewEnumerator3
    Uniform Sprite Size=50
    UsePositionOffset=true
    Write Material Random=true
   )]
  particleUpdateScript [
   &ParticleState()
   &SubUVAnimation(
    Direct Frame Index:NiagaraFloat=&Lerp_Float(A=3, B=6, Alpha=$Particles.NormalizedAge)
    "Sprite Renderer "=DataInterface() {propertyValues {SpriteRenderer=&self:Burst_Shatter_Halo001_0.NiagaraSpriteRendererProperties_0}}
    SubUV Animation Mode:ENiagara_SubUVLookupModeV2=::NewEnumerator5
   )
   &SubUVAnimation(
    Direct Frame Index:NiagaraFloat=&Lerp_Float(
     A=6
     B=9
     Alpha=&ClampFloat(Float=&Multiply_Float(A=$Particles.NormalizedAge, B=1.4))
    )
    Output Alternate Renderer Binding=true
    "Sprite Renderer "=DataInterface() {propertyValues {SpriteRenderer=&self:Burst_Shatter_Halo001_0.NiagaraSpriteRendererProperties_2}}
    SubUV Animation Mode:ENiagara_SubUVLookupModeV2=::NewEnumerator5
   ) {module=SubUVAnimation001}
   /SetVariables(
    Particles.Linear Color_Halo:LinearColor=&Multiply_LinearColorByFloat(
     Linear Color=&ColorFromCurve(ColorCurve=DataInterface() {propertyValues {
        AlphaCurve {keys [<at 0.9445 1 RCIM_Cubic>,<at 1 0 RCIM_Cubic>]}
        BlueCurve {keys [<at 0 10>,<at 0.4033 1.5>,<at 0.4132 0>,<at 0.9958 0>]}
        GreenCurve {keys [<at 0 6.2348>,<at 0.4033 0.9102>,<at 0.4132 0.6605>,<at 0.9958 0.4348>]}
        RedCurve {keys [<at 0 0.4241>,<at 0.4033 0>,<at 0.4132 1>,<at 0.9958 1>]}
       }})
     Float=$System.FrontEnd_FloatDim
    )
    Particles.SpriteSize_Halo:Vector2f=&Multiply_Vector2D_ByFloat(Float=$System.FrontEnd_Size, Vector2D=&Lerp_Vector2(A=<25 25>, B=<250 250>, Alpha=$Particles.NormalizedAge))
   ) {isSetParameters=true, module=SetVariables_#5}
   /SetVariables(
    Particles.LinearColor_Electric:LinearColor=&Multiply_LinearColorByFloat(
     Linear Color=&ColorFromCurve(ColorCurve=DataInterface() {propertyValues {
        AlphaCurve {keys [<at 0.2984 1 RCIM_Cubic>,<at 1 0>]}
        BlueCurve {keys [<at 0.0586 1 RCIM_Constant>,<at 0.1939 1 RCIM_Constant>,<at 0.3325 1 RCIM_Constant>,<at 0.493 1 RCIM_Constant>]}
        GreenCurve {keys [<at 0.0586 1 RCIM_Constant>,<at 0.1939 1 RCIM_Constant>,<at 0.3325 1 RCIM_Constant>,<at 0.493 1 RCIM_Constant>]}
        RedCurve {keys [<at 0.0586 1 RCIM_Constant>,<at 0.1939 1 RCIM_Constant>,<at 0.3325 1 RCIM_Constant>,<at 0.493 1 RCIM_Constant>]}
       }})
     Float=$System.FrontEnd_FloatDim
    )
    Particles.SpriteSize_Electric:Vector2f=&Multiply_Vector2D_ByFloat(
     Float=$System.FrontEnd_Size
     Vector2D=&Lerp_Vector2(
      A=<40 40>
      B=<80 80>
      Alpha=&FloatFromCurve(FloatCurve=DataInterface() {propertyValues {Curve {keys [
           {arriveTangent=1.98, interpMode=RCIM_Cubic, leaveTangent=1.98, tangentMode=RCTM_User, time=0, value=0}
           {arriveTangent=-0.1124, interpMode=RCIM_Cubic, leaveTangent=-0.1124, tangentMode=RCTM_User, time=1, value=1}
          ]}}})
     )
    )
   ) {isSetParameters=true}
   /SetVariables(
    Particles.Color_Shatter:LinearColor=&ColorFromCurve(
     ColorCurve=DataInterface() {propertyValues {
       AlphaCurve {keys [<at 0.8063 1 RCIM_Cubic>,<at 1 0 RCIM_Cubic>]}
       BlueCurve {keys [<at 0.5115 0.7529 RCIM_Cubic>]}
       GreenCurve {keys [<at 0.5115 0.6514 RCIM_Cubic>]}
       RedCurve {keys [<at 0.5115 0.6105 RCIM_Cubic>]}
      }}
     CurveIndex=&Multiply_Float(A=$Particles.NormalizedAge, B=1.4)
    )
    Particles.SpriteSize_Shatter:Vector2f=&Multiply_Vector2D_ByFloat(
     Float=$System.FrontEnd_Size
     Vector2D=&Lerp_Vector2(A=<40 40>, B=<60 60>, Alpha=&Multiply_Float(A=$Particles.NormalizedAge, B=1.4))
    )
   ) {isSetParameters=true}
  ]
  rendererClasses [&NiagaraSpriteRendererProperties]
  renderers [
   {rendererClass=&NiagaraSpriteRendererProperties}
   {rendererClass=&NiagaraSpriteRendererProperties, rendererIndex=1}
   {rendererClass=&NiagaraSpriteRendererProperties, rendererIndex=2}
  ]
 }
 Intro_Bolt {
  emitterUpdateScript [
   &EmitterState(Life Cycle Mode=::NewEnumerator1, Loop Behavior=::NewEnumerator1, Loop Duration=1, UseLoopDelay=false)
   &SpawnBurst_Instantaneous(
    Spawn Time=&Multiply_Float(A=$System.IntroLength, B=0.7)
    Use Spawn Probability=false
    UseLoopCountLimit=false
   )
  ]
  particleSpawnScript [&InitializeParticle(
    Color Mode=::NewEnumerator0
    Lifetime=0.15
    Position:Vector3f=$System.BurstPosition
    Position Mode=::NewEnumerator1
    Position Offset=&Multiply_VectorByFloat(Vector=<0 0 -15>, Float=$System.FrontEnd_Size)
    Sprite Size Mode=::NewEnumerator3
    Uniform Sprite Size=&Multiply_Float(A=20, B=$System.FrontEnd_Size)
    UsePositionOffset=true
    Write Material Random=true
   )]
  particleUpdateScript [
   &ParticleState()
   &ScaleColor(
    Scale RGB=&VectorFromFloat(Value=&FloatFromCurve(FloatCurve=DataInterface() {propertyValues {Curve {keys [<at 0 1 RCIM_Constant>,<at 0.7 0 RCIM_Constant>]}}}))
    ScaleA=false
    ScaleRGB=true
   )
   &SubUVAnimation(
    "Sprite Renderer "=DataInterface() {propertyValues {SpriteRenderer=&self:Intro_Bolt_0.NiagaraSpriteRendererProperties_0}}
    SubUV Lookup Index:NiagaraFloat=&Lerp_Float(A=0, B=0.33, Alpha=$Particles.NormalizedAge)
    UseEndFrame=false
    UseStartFrame=false
   )
   /SetVariables(
    Particles.Linear Color_Halo:LinearColor=&MakeLinearColorFromVectorAndFloat(
     "Vector (RGB)"=<0 0.2 0.5>
     "Float (Alpha)"=&FloatFromCurve(FloatCurve=DataInterface() {propertyValues {Curve {keys [<at 0 1>,<at 0.25 1>,<at 0.75 1>,<at 1 0>]}}}, Scale Curve=0.2)
    )
    Particles.Position_Halo:NiagaraPosition=&Lerp_Vector(
     A=&AddVectorToPosition(Position=$Particles.Position, Vector=<0 0 0>)
     B=Linked() {linkedVariable {name=Particles.Position, type=NiagaraPosition}}
     Alpha=$Particles.NormalizedAge
    )
    Particles.SpriteSize_Halo:Vector2f=&Multiply_Vector2D_ByFloat(
     Float=&FloatFromCurve(FloatCurve=DataInterface() {propertyValues {Curve {
         defaultValue=0
         keys [
          {arriveTangent=-0.0895, interpMode=RCIM_Cubic, leaveTangent=-0.0895, tangentMode=RCTM_User, time=0, value=0.5}
          {
           arriveTangent=1.4749
           arriveTangentWeight=0.1836
           interpMode=RCIM_Cubic
           leaveTangent=1.4749
           tangentMode=RCTM_User
           time=0.6496
           value=0.875
          }
          {arriveTangent=0.0101, arriveTangentWeight=0.1034, interpMode=RCIM_Cubic, leaveTangent=0.0101, tangentMode=RCTM_User, time=0.9598, value=1}
          <at 0.9931 3 RCIM_Cubic>
         ]
        }}})
     Vector2D=<25 25>
    )
   ) {isSetParameters=true, module=SetVariables_#5}
   &SpriteFacingAndAlignment(Sprite Alignment=<0 0 1>, WriteAlignment=true, WriteFacing=false)
  ]
  rendererClasses [&NiagaraSpriteRendererProperties]
  renderers [
   {rendererClass=&NiagaraSpriteRendererProperties}
   {rendererClass=&NiagaraSpriteRendererProperties, rendererIndex=1}
  ]
 }
 Intro_Heart {
  emitterSpawnScript [
   &InitializeAudioPlayer(Audio Player=DataInterface() {propertyValues {
      Attenuation=null
      Concurrency=null
      MaxPlaysPerTick=10
      SoundToPlay=/CosmeticCompanions/Assets/Audio_Shared/Interaction/Petting/SC_Mimosa_Interaction_React_Default
      bAllowLoopingOneShotSounds=false
      bLimitPlaysPerTick=true
      bOnlyActiveDuringGameplay=false
      bStopWhenComponentIsDestroyed=true
      ConfigurationUserParameter {parameter {name=None, typeDefHandle {registeredTypeIndex=1}, varData []}}
      ParameterNames []
     }})
   /Niagara/Modules/Audio/PlayAudio() {module=PlayAudio001}
  ]
  emitterUpdateScript [
   &EmitterState(Life Cycle Mode=::NewEnumerator1, Loop Behavior=::NewEnumerator1, Loop Duration=1, UseLoopDelay=false)
   &SpawnBurst_Instantaneous(Use Spawn Probability=false, UseLoopCountLimit=false)
  ]
  particleSpawnScript [
   &InitializeParticle(
    Lifetime=$System.IntroLength
    Sprite Rotation Mode=::NewEnumerator0
    Sprite Size=<15 15>
    Sprite Size Mode=::NewEnumerator4
    UsePositionOffset=false
    Write Material Random=true
   )
   &ApplyInitialForces()
  ]
  particleUpdateScript [
   &ParticleState()
   &ScaleSpriteSize(
    Scale Sprite Size Mode=::NewEnumerator0
    Uniform Scale Factor=&Add_Float(
     A=0.5
     B=&FloatFromCurve(
      FloatCurve=DataInterface() {propertyValues {Curve {
         defaultValue=0
         keys [
          {arriveTangent=-0.0895, interpMode=RCIM_Cubic, leaveTangent=-0.0895, tangentMode=RCTM_User, time=0.0989, value=0.125}
          {arriveTangent=1.4749, interpMode=RCIM_Cubic, leaveTangent=1.4749, tangentMode=RCTM_User, time=0.6496, value=0.875}
          {arriveTangent=0.0101, interpMode=RCIM_Cubic, leaveTangent=0.0101, tangentMode=RCTM_User, time=0.9598, value=1}
          <at 1 1 RCIM_Cubic>
         ]
        }}}
      Scale Curve=$System.FrontEnd_Size
     )
    )
   ) {module=ScaleSpriteSize002}
   &OffsetPosition(Position Offset=&Lerp_Vector(
     A=$System.HeartIntroSpawnPosition
     B=$System.BurstPosition
     Alpha=&FloatFromCurve(FloatCurve=DataInterface() {propertyValues {Curve {defaultValue=0, keys [{arriveTangent=2.9288, interpMode=RCIM_Cubic, leaveTangent=2.9288, tangentMode=RCTM_User, time=0, value=0}, <at 1 1>]}}})
    ))
   &DynamicMaterialParameters(
    Index 0 Param 1=1
    Index 0 Param 2=&FloatFromCurve(FloatCurve=DataInterface() {propertyValues {Curve {keys [<at 0 1>,<at 0.5 1>]}}})
    Index 0 Param 3=1
    Index 0 Param 4=&FloatFromCurve(FloatCurve=DataInterface() {propertyValues {Curve {keys [<at 0.8 0>,<at 1 1>]}}})
    Module.Index0Float0Write=true
    Module.Index0Float1Write=true
    Module.Index0Float2Write=false
    Module.Index0Float3Write=true
   )
   /SetVariables(
    Particles.Linear Color_Halo:LinearColor=&MakeLinearColorFromVectorAndFloat(
     "Vector (RGB)"=&Multiply_VectorByFloat(
      Vector=<0 0.5 1>
      Float=&FloatFromCurve(FloatCurve=DataInterface() {propertyValues {Curve {keys [<at 0 0.5>,<at 0.6 0.5>,<at 0.7 1>]}}})
     )
     "Float (Alpha)"=0.1
    )
    Particles.SpriteSize_Halo:Vector2f=&Multiply_Vector2D_ByFloat(
     Float=&FloatFromCurve(
      FloatCurve=DataInterface() {propertyValues {Curve {
         defaultValue=0
         keys [
          <at 0.6 0 RCIM_Cubic>
          {
           arriveTangent=1.4749
           arriveTangentWeight=0.1836
           interpMode=RCIM_Cubic
           leaveTangent=1.4749
           tangentMode=RCTM_User
           time=0.6496
           value=0.875
          }
          {arriveTangent=0.0101, arriveTangentWeight=0.1034, interpMode=RCIM_Cubic, leaveTangent=0.0101, tangentMode=RCTM_User, time=0.95, value=1}
         ]
        }}}
      Scale Curve=$System.FrontEnd_Size
     )
     Vector2D=<62.5 62.5>
    )
   ) {isSetParameters=true}
  ]
  rendererClasses [&NiagaraSpriteRendererProperties]
  renderers [
   {rendererClass=&NiagaraSpriteRendererProperties}
   {rendererClass=&NiagaraSpriteRendererProperties, rendererIndex=1}
  ]
 }
 Minimal_light {
  bEnabled=false
  emitterUpdateScript [
   &EmitterState(Life Cycle Mode=::NewEnumerator1, Loop Behavior=::NewEnumerator1, Loop Duration=2, UseLoopDelay=false)
   &SpawnBurst_Instantaneous(
    Spawn Time=&Multiply_Float(A=$System.IntroLength, B=0.4)
    Use Spawn Probability=false
    UseLoopCountLimit=false
   )
  ]
  particleSpawnScript [&InitializeParticle(
    Color=<rgba 10 10 10 1>
    Lifetime=&Add_Float(A=&Multiply_Float(A=$System.IntroLength, B=0.375), B=0.5)
    UsePositionOffset=false
    Write Material Random=true
   )]
  particleUpdateScript [
   &ParticleState()
   &ScaleColor(
    Scale RGB=&MakeVectorFromLinearColorRGB(Color=&ColorFromCurve(ColorCurve=DataInterface() {propertyValues {
        AlphaCurve {keys [<at 0.5927 1 RCIM_Cubic>,<at 1 1 RCIM_Cubic>]}
        BlueCurve {keys [
          <at 0.233 0.15>
          <at 0.2519 0.2>
          <at 0.6285 0.5>
          <at 0.6526 0.75>
          <at 0.6939 0.75>
          <at 0.804 0.2969>
          {arriveTangent=-1.5143, interpMode=RCIM_Cubic, leaveTangent=-1.5143, time=0.8882, value=0}
          <at 1 0>
         ]}
        GreenCurve {keys [
          <at 0.233 0.0714>
          <at 0.2519 0.1697>
          <at 0.6285 0.4242>
          <at 0.6526 0.638>
          <at 0.6939 0.6886>
          <at 0.804 0.9122>
          {arriveTangent=-4.6528, interpMode=RCIM_Cubic, leaveTangent=-4.6528, time=0.8882, value=0.3724}
          <at 1 0>
         ]}
        RedCurve {keys [
          <at 0.233 0>
          <at 0.2519 0.0547>
          <at 0.6285 0.1369>
          <at 0.6526 0.207>
          <at 0.6939 0.2774>
          <at 0.804 1>
          {arriveTangent=-5.1008, interpMode=RCIM_Cubic, leaveTangent=-5.1008, time=0.8882, value=0.75}
          <at 1 0>
         ]}
       }}))
    ScaleA=true
    ScaleRGB=true
   )
  ]
  rendererClasses [&NiagaraLightRendererProperties]
  renderers [{rendererClass=&NiagaraLightRendererProperties, rendererIndex=0}]
 }
 Shatter_Glass_Gravity {
  emitterUpdateScript [
   &EmitterState(Life Cycle Mode=::NewEnumerator1, Loop Behavior=::NewEnumerator1, Loop Duration=2, UseLoopDelay=false)
   &SpawnBurst_Instantaneous(Spawn Count=20, Spawn Time=$System.IntroLength, Use Spawn Probability=false, UseLoopCountLimit=false)
  ]
  particleSpawnScript [
   &InitializeParticle(
    Color Mode=::NewEnumerator0
    Lifetime=&RandomRangeFloat(Minimum=0.25, Maximum=0.75)
    Position:Vector3f=$System.BurstPosition
    Position Mode=::NewEnumerator1
    Sprite Rotation Mode=::NewEnumerator2
    Sprite Size Mode=::NewEnumerator3
    Sprite UV Mode=::NewEnumerator3
    Uniform Sprite Size=&RandomRangeFloat(Minimum=1, Maximum=&Multiply_Float(A=12, B=$System.FrontEnd_Size))
    UsePositionOffset=false
    Write Material Random=true
   )
   &AddVelocity(
    Cone Angle=270
    Cone Axis=<0 0 1>
    Module.UseConeFalloff=false
    Module.UseVelDistribution=false
    Override Seed=false
    Rotation Coordinate Space=::World
    Velocity Mode=::NewEnumerator2
    Velocity Speed=&Multiply_Float(A=300, B=$System.FrontEnd_Size)
   )
   &ShapeLocation(
    MaskSpawnGroup=false
    Override Seed=false
    Sphere Radius=&Multiply_Float(A=10, B=$System.FrontEnd_Size)
    UseOwnerScale=true
   )
  ]
  particleUpdateScript [
   &ParticleState()
   &ScaleSpriteSize(Uniform Curve Sprite Scale=DataInterface() {propertyValues {Curve {keys [{tangentMode=RCTM_User, time=0.5, value=1}, {tangentMode=RCTM_User, time=1, value=0}]}}})
   &SpriteRotationRate(Rotation Rate=&RandomRangeFloat(Minimum=-1000, Maximum=1000))
   &SubUVAnimation(
    Direct Frame Index=321
    "Sprite Renderer "=DataInterface() {propertyValues {SpriteRenderer=&self:Shatter_Glass_Gravity_0.NiagaraSpriteRendererProperties_2}}
    SubUV Animation Mode:ENiagara_SubUVLookupModeV2=::NewEnumerator5
   )
   &Drag(Drag=&Lerp_Float(A=2, B=4, Alpha=$Particles.Age), Use Linear Drag=true, Use Rotational Drag=true)
   &GravityForce(Gravity=<0 0 -400>)
   &SolveForcesAndVelocity(Clamp Velocity=false, Limit Acceleration=false)
   &ScaleColor(
    Scale RGB=&MakeVectorFromLinearColorRGB(Color=&ColorFromCurve(
      ColorCurve=DataInterface() {propertyValues {
        AlphaCurve {keys [<at 0.8063 1 RCIM_Cubic>,<at 1 0 RCIM_Cubic>]}
        BlueCurve {keys [<at 0.5115 0.7529 RCIM_Cubic>]}
        GreenCurve {keys [<at 0.5115 0.6514 RCIM_Cubic>]}
        RedCurve {keys [<at 0.5115 0.6105 RCIM_Cubic>]}
       }}
      CurveIndex=&Multiply_Float(A=$Particles.NormalizedAge, B=1.4)
     ))
    ScaleA=true
    ScaleRGB=true
   )
  ]
  rendererClasses [&NiagaraSpriteRendererProperties]
  renderers [{rendererClass=&NiagaraSpriteRendererProperties}]
 }
 Sparks {
  simTarget=GPUComputeSim
  emitterUpdateScript [
   &EmitterState(Life Cycle Mode=::NewEnumerator1, Loop Behavior=::NewEnumerator1, Loop Duration=2, UseLoopDelay=false)
   &SpawnBurst_Instantaneous(
    Spawn Count=30
    Spawn Time=&Add_Float(A=$System.IntroLength, B=0.1)
    Use Spawn Probability=false
    UseLoopCountLimit=false
   )
  ]
  particleSpawnScript [
   &InitializeParticle(
    Color=<rgba 10 10 10 1>
    Lifetime=&RandomRangeFloat(Minimum=0, Maximum=0.5)
    Position:Vector3f=$System.BurstPosition
    Position Mode=::NewEnumerator1
    Sprite Size Max=<0.7 5>
    Sprite Size Min=<0.5 3>
    Sprite Size Mode=::NewEnumerator2
    Sprite UV Mode=::NewEnumerator3
    UsePositionOffset=false
    Write Material Random=true
   )
   &AddVelocity(
    Module.Use Falloff Distance=false
    Override Seed=false
    Velocity Mode=::NewEnumerator1
    Velocity Origin:Vector3f=$System.BurstPosition
    Velocity Speed=&RandomRangeFloat(Minimum=50, Maximum=150)
   ) {enabled=false}
   &AddVelocity(
    Cone Angle=270
    Cone Axis=<0 0 1>
    Module.UseConeFalloff=false
    Module.UseVelDistribution=false
    Override Seed=false
    Rotation Coordinate Space=::World
    Velocity Mode=::NewEnumerator2
    Velocity Speed=&Multiply_Float(A=&RandomRangeFloat(Minimum=100, Maximum=200), B=$System.FrontEnd_Size)
   ) {module=AddVelocity001}
   &ShapeLocation(MaskSpawnGroup=false, Override Seed=false, Sphere Radius=3, UseOwnerScale=true)
  ]
  particleUpdateScript [
   &ParticleState()
   &Drag(Drag=&Lerp_Float(A=2, B=4, Alpha=$Particles.Age), Use Linear Drag=true, Use Rotational Drag=false)
   &GravityForce(Gravity=<0 0 -700>)
   &SolveForcesAndVelocity(Clamp Velocity=false, Limit Acceleration=false)
   &ScaleSpriteSizeBySpeed(Max Scale Factor=<1 2>)
   &ScaleSpriteSize(
    Scale Factor=&Multiply_Vector2D_ByFloat(Float=$System.FrontEnd_Size, Vector2D=&Lerp_Vector2(A=<1 1>, B=<1 0.5>, Alpha=$Particles.NormalizedAge))
    Scale Sprite Size Mode=::NewEnumerator2
   )
   &ScaleColor(
    Scale Alpha=$System.FrontEnd_FloatDim
    Scale RGB=&MakeVectorFromLinearColorRGB(Color=&ColorFromCurve(ColorCurve=DataInterface() {propertyValues {
        AlphaCurve {keys [<at 0.4039 1 RCIM_Cubic>,<at 1 0 RCIM_Cubic>]}
        BlueCurve {keys [<at 0.2494 1>, {arriveTangent=-1.3322, interpMode=RCIM_Cubic, leaveTangent=-1.3322, time=0.4652, value=0.0885}, <at 0.745 0.201>, <at 1 0>]}
        GreenCurve {keys [<at 0.2494 1>, {arriveTangent=-1.3322, interpMode=RCIM_Cubic, leaveTangent=-1.3322, time=0.4652, value=0.9679}, <at 0.745 0.5635>, <at 1 0>]}
        RedCurve {keys [<at 0.2494 1>, {arriveTangent=-1.3322, interpMode=RCIM_Cubic, leaveTangent=-1.3322, time=0.4652, value=1}, <at 0.745 1>, <at 1 0>]}
       }}))
    ScaleA=true
    ScaleRGB=true
   )
   /SetVariables(
    Particles.Linear Color_Halo:LinearColor=&Multiply_LinearColorByFloat(
     Linear Color=&ColorFromCurve(ColorCurve=DataInterface() {propertyValues {
        AlphaCurve {keys [<at 0.4039 0.25 RCIM_Cubic>,<at 1 0 RCIM_Cubic>]}
        BlueCurve {keys [<at 0.2494 1>, {arriveTangent=-1.612, interpMode=RCIM_Cubic, leaveTangent=-1.612, time=0.4652, value=0.0885}, <at 0.745 0.201>, <at 1 0>]}
        GreenCurve {keys [<at 0.2494 1>, {arriveTangent=-0.8807, interpMode=RCIM_Cubic, leaveTangent=-0.8807, time=0.4652, value=0.9679}, <at 0.745 0.5635>, <at 1 0>]}
        RedCurve {keys [<at 0.2494 1>,<at 0.4652 1 RCIM_Cubic>,<at 0.745 1>,<at 1 0>]}
       }})
     Float=$System.FrontEnd_FloatDim
    )
    Particles.SpriteSize_Halo:Vector2f=&Multiply_Vector2D_ByFloat(Float=$System.FrontEnd_Size, Vector2D=<5 10>)
   ) {isSetParameters=true}
  ]
  rendererClasses [&NiagaraSpriteRendererProperties]
  renderers [
   {rendererClass=&NiagaraSpriteRendererProperties}
   {rendererClass=&NiagaraSpriteRendererProperties, rendererIndex=1}
  ]
 }
 lightningBolts {
  emitterUpdateScript [
   &EmitterState(Life Cycle Mode=::NewEnumerator1, Loop Behavior=::NewEnumerator1, Loop Duration=2, UseLoopDelay=false)
   &SpawnBurst_Instantaneous(
    Spawn Count=11
    Spawn Time=&Add_Float(A=$System.IntroLength, B=0.1)
    Use Spawn Probability=false
    UseLoopCountLimit=false
   )
  ]
  particleSpawnScript [
   &InitializeParticle(
    Color=&SelectLinearColorFromArray(
     Array Sampling Mode=::NewEnumerator1
     Direct Array Index=&RandomRangeInt(Minimum=0, Maximum=1)
     Color Selection Array=DataInterface() {propertyValues {ColorData [<rgba 5 1 0.5 1>, <rgba 1 1 1 1>]}}
     Random Seed=?
    )
    Lifetime=&RandomRangeFloat(Minimum=0.125, Maximum=0.25)
    Position:Vector3f=$System.BurstPosition
    Position Mode=::NewEnumerator1
    Sprite Size=&Multiply_Vector2D_ByFloat(Float=$System.FrontEnd_Size, Vector2D=&RandomRangeVector2D(Minimum=<7.5 10>, Maximum=<15 20>))
    Sprite Size Mode=::NewEnumerator4
    Sprite UV Mode=::NewEnumerator1
    UsePositionOffset=false
    Write Material Random=true
   )
   &ShapeLocation(MaskSpawnGroup=false, Override Seed=false, Sphere Radius=15, Sphere Surface Distribution=1, UseOwnerScale=true)
   &AddVelocity(
    Module.Use Falloff Distance=false
    Override Seed=false
    Velocity Mode=::NewEnumerator1
    Velocity Origin:Vector3f=$System.BurstPosition
    Velocity Speed=&Multiply_Float(A=50, B=$System.FrontEnd_Size)
   )
  ]
  particleUpdateScript [
   &ParticleState()
   &ScaleSpriteSize(
    Scale Factor=&Lerp_Vector2(A=<1 1>, B=<1 3>, Alpha=$Particles.NormalizedAge)
    Scale Sprite Size Mode=::NewEnumerator2
   )
   &SubUVAnimation(
    Direct Frame Index:NiagaraFloat=/Niagara/DynamicInputs/Math/Floor_Float(Float:NiagaraFloat=&Lerp_Float(
      A=0
      B=2.99
      Alpha=&ClampFloat(Float=&Multiply_Float(A=$Particles.NormalizedAge, B=1))
     ))
    "Sprite Renderer "=DataInterface() {propertyValues {SpriteRenderer=&self:lightningBolts_0.NiagaraSpriteRendererProperties_0}}
    SubUV Animation Mode:ENiagara_SubUVLookupModeV2=::NewEnumerator5
   )
   /SetVariables(
    Particles.Linear Color_Halo:LinearColor=&MakeLinearColorFromVectorAndFloat(
     "Vector (RGB)"=<0 0.2 0.5>
     "Float (Alpha)"=&FloatFromCurve(FloatCurve=DataInterface() {propertyValues {Curve {keys [<at 0 1>,<at 0.25 1>,<at 0.75 1>,<at 1 0>]}}}, Scale Curve=0.5)
    )
    Particles.Position_Halo:NiagaraPosition=&Lerp_Vector(
     A=&AddVectorToPosition(Position=$Particles.Position, Vector=<0 0 -8>)
     B=Linked() {linkedVariable {name=Particles.Position, type=NiagaraPosition}}
     Alpha=$Particles.NormalizedAge
    )
    Particles.SpriteSize_Halo:Vector2f=<25 25>
   ) {enabled=false, isSetParameters=true}
   &SolveForcesAndVelocity(Clamp Velocity=false, Limit Acceleration=false)
   &CameraOffset(Camera Offset Amount=-5)
   &ScaleColor(
    Scale Alpha=&FloatFromCurve(FloatCurve=DataInterface() {propertyValues {Curve {keys [<at 0.75 1>,<at 1 0>]}}})
    ScaleA=true
    ScaleRGB=true
   )
  ]
  rendererClasses [&NiagaraSpriteRendererProperties]
  renderers [
   {rendererClass=&NiagaraSpriteRendererProperties}
   {rendererClass=&NiagaraSpriteRendererProperties, rendererIndex=1}
  ]
 }
}
summary {
 systemName=NS_ReactFX_BirthMitten
 emitters [
  {emitterName=Intro_Heart, rendererClasses [&NiagaraSpriteRendererProperties]}
  {emitterName=Intro_Bolt, rendererClasses [&NiagaraSpriteRendererProperties]}
  {emitterName=Burst_Shatter_Halo001, rendererClasses [&NiagaraSpriteRendererProperties]}
  {emitterName=Shatter_Glass_Gravity, rendererClasses [&NiagaraSpriteRendererProperties]}
  {emitterName=lightningBolts, rendererClasses [&NiagaraSpriteRendererProperties]}
  {emitterName=Sparks, simTarget=GPUComputeSim, rendererClasses [&NiagaraSpriteRendererProperties]}
  {bEnabled=false, emitterName=Minimal_light, rendererClasses [&NiagaraLightRendererProperties]}
 ]
 userVariables [
  User.bIsFrontEnd {type=NiagaraBool}
  User.bIsFrontEndPreview {type=NiagaraBool}
 ]
}
