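# Niagara reaction effect for the BriskImp cosmetic companion: an icy heart with a petting-reaction sound, ending in a frosty burst.
# A swirling vortex burst opens it, while helical ribbons and smoke travel from the intro spawn position toward the burst position.
# Heart and glow follow a system offset position; smoke and skeletal-mesh-sampled ice fragments burst just before intro length ends.
# Time of day tints colors, and front-end scale resizes several emitters.
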NiagaraSystem /CosmeticCompanions/FX/React/NS_ReactFX_BriskImp
dependencies {
 usedDataInterfaces [
  &NiagaraDataInterfaceAudioPlayer
  &NiagaraDataInterfaceCurve
  &NiagaraDataInterfaceColorCurve
  &NiagaraDataInterfaceMeshRendererInfo
  &NiagaraDataInterfaceSpriteRendererInfo
 ]
 usedDynamicInputs [
  &FN_FrontEndVector3Modifier
  &Boolean_OrOperation
  &MakeFloatFromVector
 ]
 usedModules [
  &self:SystemSpawnScript.SystemScriptSource.SystemScriptGraph.NiagaraNodeAssignment_0.SetVariables_#1
  &self:SystemSpawnScript.SystemScriptSource.SystemScriptGraph.NiagaraNodeAssignment_6.SetVariables_#2
  &self:SystemSpawnScript.SystemScriptSource.SystemScriptGraph.NiagaraNodeAssignment_7.SetVariables_#3
  &self:SystemSpawnScript.SystemScriptSource.SystemScriptGraph.NiagaraNodeAssignment_5.SetVariables_#4
  &self:SystemSpawnScript.SystemScriptSource.SystemScriptGraph.NiagaraNodeAssignment_3.SetVariables_#5
  &self:SystemSpawnScript.SystemScriptSource.SystemScriptGraph.NiagaraNodeAssignment_1.SetVariables_#6
  &SystemState
  &self:SystemSpawnScript.SystemScriptSource.SystemScriptGraph.NiagaraNodeAssignment_8.SetVariables_#7
 ]
}
emitters {
 Frosty_Smoke {
  emitterUpdateScript [
   &EmitterState()
   &SpawnRate(
    SpawnRate=&FloatFromCurve(
     FloatCurve=DataInterface() {propertyValues {Curve {keys [{arriveTangent=3.0606, interpMode=RCIM_Constant, time=0, value=0}, <at 0.1 1 RCIM_Constant>, <at 0.5 0 RCIM_Constant>]}}}
     CurveIndex=&Divide_Float(A=$Emitter.Age, B=$System.IntroLength)
     Scale Curve=20
    )
    Use Spawn Probability=false
   )
   &SpawnBurst_Instantaneous(
    Spawn Count=8
    Spawn Group=1
    Spawn Time=&Add_Float(A=$System.IntroLength, B=-0.2)
    Use Spawn Probability=false
    UseLoopCountLimit=false
   )
  ]
  particleSpawnScript [
   &InitializeParticle(
    Color=&MultiplyLinearColor(
     Linear Color=&RandomRangeLinearColor(Minimum=<rgba 0.3906 0.7302 1 1>, Maximum=<rgba 0.5469 0.7994 1 1>)
     Scale Factor=&MakeVector4_FromLinearColor(LinearColor=$System.TOD)
    )
    Lifetime=&RandomRangeFloat(Minimum=0.8, Maximum=1.5)
    Sprite Rotation Mode=::NewEnumerator2
    Sprite Size Mode=::NewEnumerator3
    Uniform Sprite Size=&Multiply_Float(A=30, B=&FN_FrontEndFloatModifier(Gameplay=1, Front End=$User.FrontEndScale))
    UsePositionOffset=false
    Write Material Random=true
   ) {module=InitializeParticle001}
   &CameraOffset(Camera Offset Amount=20)
   /SetVariables(Particles.Position=&Lerp_Vector(
     A=$System.HeartIntroSpawnPosition
     B=$System.BurstPosition
     Alpha=&FloatFromCurve(
      FloatCurve=DataInterface() {propertyValues {Curve {keys [{arriveTangent=2.5323, interpMode=RCIM_Cubic, leaveTangent=2.5323, tangentMode=RCTM_User, time=0, value=0}, <at 1 1 RCIM_Cubic>]}}}
      CurveIndex=&Add_Float(A=0.1, B=&Divide_Float(A=$Engine.System.Age, B=$System.IntroLength))
     )
    )) {isSetParameters=true}
   &AddVelocity(
    Velocity=&RandomRangeVector(Minimum=<-1 -1 -1>, Maximum=<1 1 1>)
    Velocity Speed Scale=&Multiply_Float(
     A=&MergeHelper_FloatSwitch(0=1, 1=50, 2=?, 3=?, Index=&SpawnGroup(), Num=::NewEnumerator0)
     B=&FN_FrontEndFloatModifier(Gameplay=1, Front End=$User.FrontEndScale)
    )
   )
   /SetVariables(Particles.StartIndex:NiagaraInt32=&RandomRangeInt(Minimum=2, Maximum=20)) {isSetParameters=true}
  ]
  particleUpdateScript [
   &ParticleState()
   &SubUVAnimation(
    End Frame:NiagaraInt32=&Add_Integer(A=$Particles.StartIndex, B=&RandomRangeInt(Minimum=20, Maximum=30))
    Start Frame:NiagaraInt32=$Particles.StartIndex
   )
   &GravityForce(Gravity=&RandomRangeVector(Minimum=<0 0 -6>, Maximum=<0 0 -10>))
   &Drag(Drag=&RandomRangeFloat(Minimum=2, Maximum=3), Use Linear Drag=true, Use Rotational Drag=true)
   &SolveForcesAndVelocity(Clamp Velocity=false, Limit Acceleration=false)
   &ScaleColor(
    Scale Alpha=&FloatFromCurve(
     FloatCurve=DataInterface() {propertyValues {Curve {keys [
         <at 0 0 RCIM_Cubic>
         {arriveTangent=0.317, interpMode=RCIM_Cubic, leaveTangent=0.317, tangentMode=RCTM_User, time=0.1, value=0.4}
         <at 1 0 RCIM_Cubic>
        ]}}}
     Scale Curve=&Multiply_Float(
      A=&Multiply_Float(
       A=&RandomRangeFloat(Minimum=0.3, Maximum=0.5)
       B=&FN_FrontEndFloatModifier(Gameplay=1, Front End=0.5)
      )
      B=&FN_Select_Float_from_Time_Of_Day(Morning=0.5, Day=1, Evening=0.5, Night=0.5)
     )
    )
    Scale RGB=<1 1 1.1>
    ScaleA=true
    ScaleRGB=true
   )
   &DynamicMaterialParameters(
    Index 0 Param 4=1
    Module.Param0WriteEnabled=true
    Module.Param1WriteEnabled=true
    Module.Param2WriteEnabled=true
    Module.Param3WriteEnabled=true
   )
   &ScaleSpriteSize(Scale Factor=&Vector2DFromFloat(Value=&FloatFromCurve(FloatCurve=DataInterface() {propertyValues {Curve {keys [
          {arriveTangent=2.1241, interpMode=RCIM_Cubic, leaveTangent=2.1241, tangentMode=RCTM_User, time=0, value=0.5}
          {arriveTangent=0.8108, interpMode=RCIM_Cubic, leaveTangent=0.8108, tangentMode=RCTM_User, time=1, value=1.5}
         ]}}})))
  ]
  rendererClasses [&NiagaraSpriteRendererProperties]
  renderers [{rendererClass=&NiagaraSpriteRendererProperties}]
 }
 Heart {
  emitterSpawnScript [
   &InitializeAudioPlayer(Audio Player=DataInterface() {propertyValues {
      Attenuation=null
      Concurrency=null
      MaxPlaysPerTick=10
      SoundToPlay=/CosmeticCompanions/Assets/Audio_Shared/Interaction/Petting/SC_Mimosa_Interaction_React_Default
      bAllowLoopingOneShotSounds=false
      bLimitPlaysPerTick=true
      bOnlyActiveDuringGameplay=false
      bStopWhenComponentIsDestroyed=true
      ConfigurationUserParameter {parameter {name=None, typeDefHandle {registeredTypeIndex=1}, varData []}}
      ParameterNames []
     }})
   /Niagara/Modules/Audio/PlayAudio() {module=PlayAudio001}
  ]
  emitterUpdateScript [
   &EmitterState()
   &SpawnRate(
    SpawnGroup=1
    SpawnRate=&FloatFromCurve(FloatCurve=DataInterface() {propertyValues {Curve {keys [<at 0.49 1>,<at 0.5 0>]}}}, CurveIndex=$Emitter.NormalizedLoopAge, Scale Curve=50)
    Use Spawn Probability=false
   )
   &SpawnBurst_Instantaneous(Use Spawn Probability=false, UseLoopCountLimit=false)
  ]
  particleSpawnScript [
   /SetVariables(Particles.VisibilityTag=$Emitter.SpawnGroup) {isSetParameters=true}
   &InitializeParticle(
    Color=$System.TOD
    Lifetime=$System.IntroLength
    Sprite Rotation Mode=::NewEnumerator0
    Sprite Size Mode=::NewEnumerator3
    Uniform Sprite Size=&Multiply_Float(A=18, B=&FN_FrontEndFloatModifier(Gameplay=1, Front End=$User.FrontEndScale))
    UsePositionOffset=false
    Write Material Random=true
   )
  ]
  particleUpdateScript [
   &ParticleState()
   &ScaleSpriteSize(
    Scale Sprite Size Mode=::NewEnumerator0
    Uniform Scale Factor=&Add_Float(
     A=0
     B=&FloatFromCurve(FloatCurve=DataInterface() {propertyValues {Curve {
         defaultValue=0
         keys [
          <at 0 0.3 RCIM_Cubic>
          {arriveTangent=1.25, leaveTangent=1.25, time=0.4, value=1}
          {arriveTangent=6.566, interpMode=RCIM_Cubic, leaveTangent=6.566, tangentMode=RCTM_User, time=0.9, value=1.1}
          {arriveTangent=35.933, interpMode=RCIM_Cubic, leaveTangent=35.933, tangentMode=RCTM_User, time=1, value=3}
         ]
        }}})
    )
   ) {module=ScaleSpriteSize002}
   &ScaleColor(
    Scale Alpha=&FloatFromCurve(
     FloatCurve=DataInterface() {propertyValues {Curve {
        defaultValue=0
        keys [
         {arriveTangent=9.0725, interpMode=RCIM_Cubic, leaveTangent=9.0725, tangentMode=RCTM_User, time=0, value=0}
         {arriveTangent=-0.2441, leaveTangent=-0.2441, time=0.1, value=0.9847}
         {arriveTangent=-1.146, leaveTangent=-1.146, time=0.952, value=0.9893}
         {arriveTangent=-0.0405, interpMode=RCIM_Cubic, leaveTangent=-0.0405, tangentMode=RCTM_User, time=1, value=0}
        ]
       }}}
     Scale Curve=0.95
    )
    ScaleA=true
    ScaleRGB=true
   )
   /SetVariables(Particles.Position=$System.OffsetPosition) {isSetParameters=true}
   &DynamicMaterialParameters(
    Index 0 Param 1=&FloatFromCurve(FloatCurve=DataInterface() {propertyValues {Curve {keys [{arriveTangent=0.6296, interpMode=RCIM_Cubic, leaveTangent=0.6296, tangentMode=RCTM_User, time=0.5, value=1}, <at 1 2>]}}})
    Index 0 Param 2=&FloatFromCurve(FloatCurve=DataInterface() {propertyValues {Curve {keys [{arriveTangent=0.6296, interpMode=RCIM_Cubic, leaveTangent=0.6296, tangentMode=RCTM_User, time=0.5, value=1}, <at 0.6 50>, <at 0.7 1>]}}})
    Index 0 Param 3=1
    Module.Index0Float0Write=true
    Module.Index0Float1Write=true
    Module.Index0Float2Write=true
    Module.Index0Float3Write=true
   )
  ]
  rendererClasses [&NiagaraSpriteRendererProperties]
  renderers [{rendererClass=&NiagaraSpriteRendererProperties}]
 }
 HeartGlow {
  emitterUpdateScript [
   &EmitterState()
   &SpawnBurst_Instantaneous(Use Spawn Probability=false, UseLoopCountLimit=false)
  ]
  particleSpawnScript [&InitializeParticle(
    Color=$System.TOD
    Position Offset=$System.BurstPosition
    Sprite Size Mode=::NewEnumerator3
    Uniform Sprite Size=&Multiply_Float(A=25, B=&FN_FrontEndFloatModifier(Gameplay=1, Front End=$User.FrontEndScale))
    UsePositionOffset=true
    Write Material Random=true
   )]
  particleUpdateScript [
   &ParticleState()
   &ScaleSpriteSize(
    Scale Sprite Size Mode=::NewEnumerator0
    Uniform Scale Factor=&FloatFromCurve(
     FloatCurve=DataInterface() {propertyValues {Curve {keys [
         {arriveTangent=3.3458, interpMode=RCIM_Cubic, leaveTangent=3.3458, tangentMode=RCTM_User, time=0, value=1}
         {arriveTangent=1.251, tangentMode=RCTM_User, time=0.7, value=3}
         <at 0.9 2>
         <at 1.0056 5>
        ]}}}
     Scale Curve=&Waveform("Waveform [1]"=::NewEnumerator12, "[1] Frequency"=5, "[1] Amplitude Min/Max"=<0.7 1>)
    )
   ) {module=ScaleSpriteSize002}
   &ScaleColor(
    Scale Alpha=&FloatFromCurve(FloatCurve=DataInterface() {propertyValues {Curve {
        defaultValue=0
        keys [
         {arriveTangent=0.6582, interpMode=RCIM_Cubic, leaveTangent=0.6582, tangentMode=RCTM_User, time=0.3, value=0}
         {arriveTangent=1.6667, interpMode=RCIM_Cubic, leaveTangent=1.6667, time=0.7134, value=0.3}
         {arriveTangent=0.3137, interpMode=RCIM_Cubic, leaveTangent=0.3137, tangentMode=RCTM_User, time=0.9, value=1}
        ]
       }}})
    Scale RGB=&MakeVectorFromLinearColorRGB(Color=&ColorFromCurve(ColorCurve=DataInterface() {propertyValues {
        AlphaCurve {keys [<at 0 1 RCIM_Cubic>]}
        BlueCurve {keys [<at 0.3291 1 RCIM_Cubic>,<at 0.8239 1>,<at 1 1 RCIM_Cubic>]}
        GreenCurve {keys [<at 0.3291 0.1287 RCIM_Cubic>,<at 0.8239 0.3778>,<at 1 0.137 RCIM_Cubic>]}
        RedCurve {keys [<at 0.3291 0 RCIM_Cubic>,<at 0.8239 0>,<at 1 0 RCIM_Cubic>]}
       }}))
    ScaleA=true
    ScaleRGB=true
   )
   /SetVariables(Particles.Position=$System.OffsetPosition) {isSetParameters=true}
   &DynamicMaterialParameters(
    Index 0 Param 1=1
    Index 0 Param 3=1
    Module.Index0Float0Write=true
    Module.Index0Float1Write=true
    Module.Index0Float2Write=true
    Module.Index0Float3Write=true
   )
   &FN_Sprite_Attributes(
    Color=&MakeLinearColorFromVectorAndFloat(
     "Vector (RGB)"=&MakeVectorFromLinearColorRGB(Color=<rgba 0.5149 1 0.9911 1>)
     "Float (Alpha)"=&FloatFromCurve(
      FloatCurve=DataInterface() {propertyValues {Curve {
         defaultValue=0
         keys [
          {arriveTangent=0.6582, interpMode=RCIM_Cubic, leaveTangent=0.6582, tangentMode=RCTM_User, time=0, value=0}
          {arriveTangent=0.3137, interpMode=RCIM_Cubic, leaveTangent=0.3137, tangentMode=RCTM_User, time=0.8, value=1}
          <at 0.9832 0 RCIM_Cubic>
         ]
        }}}
      Scale Curve=0.2
     )
    )
    Create Color Attribute=true
    Create SpriteSize Attribute=true
    SpriteSize=&Vector2DFromFloat(Value=&Multiply_Float(
      A=&Multiply_Float(A=20, B=&FN_FrontEndFloatModifier(Gameplay=1, Front End=$User.FrontEndScale))
      B=&FloatFromCurve(FloatCurve=DataInterface() {propertyValues {Curve {keys [
           {arriveTangent=3.3458, interpMode=RCIM_Cubic, leaveTangent=3.3458, tangentMode=RCTM_User, time=0, value=1}
           {arriveTangent=1.251, tangentMode=RCTM_User, time=0.5, value=3}
           <at 0.8 3>
           <at 0.9809 5>
          ]}}})
     ))
    Write to Module Namespace=true
   )
  ]
  rendererClasses [&NiagaraSpriteRendererProperties]
  renderers [
   {rendererClass=&NiagaraSpriteRendererProperties}
   {rendererClass=&NiagaraSpriteRendererProperties, rendererIndex=1}
  ]
 }
 Ice_Frags {
  emitterUpdateScript [
   &EmitterState()
   &SpawnBurst_Instantaneous(
    Spawn Count=15
    Spawn Time=&Add_Float(A=$System.IntroLength, B=-0.2)
    Use Spawn Probability=false
    UseLoopCountLimit=false
   )
   &SpawnRate(
    SpawnRate=&FloatFromCurve(
     FloatCurve=DataInterface() {propertyValues {Curve {
        defaultValue=1
        keys [
         {arriveTangent=4.1785, interpMode=RCIM_Cubic, leaveTangent=4.1785, tangentMode=RCTM_User, time=0, value=0}
         {arriveTangent=1.1179, interpMode=RCIM_Cubic, leaveTangent=1.1179, tangentMode=RCTM_User, time=0.2042, value=0.6228}
         {arriveTangent=0.0224, interpMode=RCIM_Cubic, leaveTangent=0.0224, tangentMode=RCTM_User, time=1, value=1}
        ]
       }}}
     CurveIndex=$Emitter.NormalizedLoopAge
     Scale Curve=&Lerp_Float(A=40, B=80, Alpha=$User.ParaGlideLeanAlpha)
    )
    Use Spawn Probability=false
   ) {enabled=false}
  ]
  particleSpawnScript [
   &PartitionParticles()
   &InitializeParticle(
    Color=&MakeCustomLinearColorFromBool(Bool=&RandomBool(), True Color=<rgba 0.4995 1.0577 1.5 1>, False Color=<rgba 0.1213 0.3179 1.5 1>)
    Lifetime Max=0.5
    Lifetime Min=0.23
    Lifetime Mode=::NewEnumerator1
    Mesh Renderer Array Visibility Mode=::NewEnumerator1
    Mesh Renderer Info=DataInterface() {propertyValues {MeshRenderer=&self:Ice_Frags_0.NiagaraMeshRendererProperties_2}}
    Mesh Renderer Mesh Index:NiagaraInt32=&RandomRangeInt(Minimum=0, Maximum=1)
    Mesh Scale Max=<0.15 0.15 0.2>
    Mesh Scale Min=<0.075 0.075 0.1>
    Mesh Scale Mode=::NewEnumerator2
    UsePositionOffset=false
    Write Material Random=true
   )
   &SkeletalMeshLocation(
    Bone / Socket Index=$Particles.PartitionParticles.Partition
    Bone Sampling Mode=::NewEnumerator4
    MaskSpawnGroup=false
    Skeletal Mesh=$System.Skeletal Mesh
   )
   &CameraOffset(Camera Offset Amount=&RandomRangeFloat(Minimum=0, Maximum=20, Evaluation Type=::NewEnumerator1))
   &ShapeLocation(MaskSpawnGroup=false, Override Seed=false, Sphere Radius=1, UseOwnerScale=true)
   &AddVelocity(
    Module.Use Falloff Distance=false
    Override Seed=false
    Velocity Mode=::NewEnumerator1
    Velocity Origin=$Particles.ShapeLocation.ShapePosition
    Velocity Speed=&RandomRangeFloat(Minimum=50, Maximum=100)
   )
  ]
  particleUpdateScript [
   &ParticleState()
   &SolveForcesAndVelocity(Clamp Velocity=false, Limit Acceleration=false)
   &ScaleMeshSize(Scale Factor=&VectorFromFloat(Value=&FloatFromCurve(FloatCurve=DataInterface() {propertyValues {Curve {
         defaultValue=0.6697
         keys [
          {tangentMode=RCTM_User, time=0, value=1}
          {arriveTangent=-0.0405, interpMode=RCIM_Cubic, leaveTangent=-0.0405, tangentMode=RCTM_User, time=0.7009, value=1}
          {arriveTangent=-8.0089, interpMode=RCIM_Cubic, leaveTangent=-8.0089, tangentMode=RCTM_User, time=1, value=0}
         ]
        }}})))
   &ScaleColor(
    Scale Alpha=&FloatFromCurve(FloatCurve=DataInterface() {propertyValues {Curve {keys [<at 0.5 1 RCIM_Cubic>,<at 1.0011 0.002 RCIM_Cubic>]}}})
    Scale RGB=&MakeVectorFromLinearColorRGB(Color=&FN_Select_Color_from_Time_Of_Day(Morning=<rgba 0.3208 0.5045 0.8 1>, Day=<rgba 0.3125 0.5928 1 1>, Evening=<rgba 0.3542 0.6556 1 1>, Night=<rgba 0.2161 0.2929 0.5 1>))
    ScaleA=true
    ScaleRGB=true
   )
  ]
  rendererClasses [&NiagaraMeshRendererProperties]
  renderers [{rendererClass=&NiagaraMeshRendererProperties, rendererIndex=0}]
 }
 Ice_Frags001 {
  emitterUpdateScript [
   &EmitterState()
   &SpawnBurst_Instantaneous(
    Spawn Count=15
    Spawn Time=&Add_Float(A=$System.IntroLength, B=-0.2)
    Use Spawn Probability=false
    UseLoopCountLimit=false
   )
   &SpawnRate(
    SpawnRate=&FloatFromCurve(
     FloatCurve=DataInterface() {propertyValues {Curve {
        defaultValue=1
        keys [
         {arriveTangent=4.1785, interpMode=RCIM_Cubic, leaveTangent=4.1785, tangentMode=RCTM_User, time=0, value=0}
         {arriveTangent=1.1179, interpMode=RCIM_Cubic, leaveTangent=1.1179, tangentMode=RCTM_User, time=0.2042, value=0.6228}
         {arriveTangent=0.0224, interpMode=RCIM_Cubic, leaveTangent=0.0224, tangentMode=RCTM_User, time=1, value=1}
        ]
       }}}
     CurveIndex=$Emitter.NormalizedLoopAge
     Scale Curve=&Lerp_Float(A=40, B=80, Alpha=$User.ParaGlideLeanAlpha)
    )
    Use Spawn Probability=false
   ) {enabled=false}
  ]
  particleSpawnScript [
   &PartitionParticles()
   &InitializeParticle(
    Color=&FNTimeOfDayColor(
     Night Color=<rgba 0.4271 0.4271 0.4271 1>
     Morning Color=<rgba 0.4635 0.7625 1 1>
     Day Color=<rgba 0.4635 0.7625 1 0.6>
     Evening Color=<rgba 0.4635 0.7625 1 1>
    )
    Lifetime Max=0.7
    Lifetime Min=0.4
    Lifetime Mode=::NewEnumerator1
    Sprite Size Mode=::NewEnumerator1
    Uniform Sprite Size Max=5
    Uniform Sprite Size Min=3
    UsePositionOffset=false
    Write Material Random=true
   )
   &SkeletalMeshLocation(
    Bone / Socket Index=$Particles.PartitionParticles.Partition
    Bone Sampling Mode=::NewEnumerator4
    MaskSpawnGroup=false
    Skeletal Mesh=$System.Skeletal Mesh
   )
   &CameraOffset(Camera Offset Amount=&RandomRangeFloat(Minimum=5, Maximum=20, Evaluation Type=::NewEnumerator1))
   &ShapeLocation(MaskSpawnGroup=false, Override Seed=false, Sphere Radius=1, UseOwnerScale=true)
   &AddVelocity(
    Module.Use Falloff Distance=false
    Override Seed=false
    Velocity Mode=::NewEnumerator1
    Velocity Origin=$Particles.ShapeLocation.ShapePosition
    Velocity Speed=&RandomRangeFloat(Minimum=50, Maximum=75)
   )
  ]
  particleUpdateScript [
   &ParticleState()
   &CurlNoiseForce(Bias Noise=false, Noise Frequency=500, Noise Strength=50, Pan Noise=false)
   &SolveForcesAndVelocity(Clamp Velocity=false, Limit Acceleration=false)
   &ScaleMeshSize(Scale Factor=&VectorFromFloat(Value=&FloatFromCurve(FloatCurve=DataInterface() {propertyValues {Curve {
         defaultValue=0.6697
         keys [
          {tangentMode=RCTM_User, time=0, value=1}
          {arriveTangent=-0.0405, interpMode=RCIM_Cubic, leaveTangent=-0.0405, tangentMode=RCTM_User, time=0.7009, value=1}
          {arriveTangent=-8.0089, interpMode=RCIM_Cubic, leaveTangent=-8.0089, tangentMode=RCTM_User, time=1, value=0}
         ]
        }}})))
   &ScaleColor(
    Scale Alpha=&FloatFromCurve(FloatCurve=DataInterface() {propertyValues {Curve {keys [<at 0.5 1 RCIM_Cubic>,<at 1.0011 0.002 RCIM_Cubic>]}}})
    Scale RGB=&MakeVectorFromLinearColorRGB(Color=&FN_Select_Color_from_Time_Of_Day(Morning=<rgba 0.3208 0.5045 0.8 1>, Day=<rgba 0.3125 0.5928 1 1>, Evening=<rgba 0.3542 0.6556 1 1>, Night=<rgba 0.2161 0.2929 0.5 1>))
    ScaleA=true
    ScaleRGB=true
   )
  ]
  rendererClasses [&NiagaraSpriteRendererProperties]
  renderers [{rendererClass=&NiagaraSpriteRendererProperties}]
 }
 LWE_AirGlow {rendererClasses [], renderers []}
 LWE_Corona {rendererClasses [], renderers []}
 LWE_Mist {rendererClasses [], renderers []}
 LWE_Rays {rendererClasses [], renderers []}
 LWE_Rays001 {rendererClasses [], renderers []}
 SpawnBurst {
  emitterUpdateScript [
   &EmitterState()
   &SpawnBurst_Instantaneous(Spawn Count=8, Spawn Group=1, Use Spawn Probability=false, UseLoopCountLimit=false) {module=SpawnBurst_Instantaneous001}
   &SpawnBurst_Instantaneous(Spawn Count=6, Use Spawn Probability=false, UseLoopCountLimit=false)
  ]
  particleSpawnScript [
   &InitializeParticle(
    Color=&FNTimeOfDayColor(
     Night Color=<rgba 0.4271 0.4271 0.4271 1>
     Morning Color=<rgba 0.4635 0.7625 1 1>
     Day Color=<rgba 0.4635 0.7625 1 0.6>
     Evening Color=<rgba 0.4635 0.7625 1 1>
    )
    Lifetime Max=1
    Lifetime Min=0.75
    Lifetime Mode=::NewEnumerator1
    Position Offset=&SubtractVectorFromPosition(Position=Linked() {linkedVariable {name=System.HeartIntroSpawnPosition, type=Vector3f}}, Vector=<0 0 50>)
    Sprite Rotation Mode=::NewEnumerator2
    Sprite Size Mode=::NewEnumerator1
    Uniform Sprite Size Max=5
    Uniform Sprite Size Min=3
    UsePositionOffset=true
    Write Material Random=true
   )
   &AddVelocity(
    Cone Axis=<0 0 1>
    Module.UseConeFalloff=false
    Module.UseVelDistribution=false
    Override Seed=false
    Velocity Mode=::NewEnumerator2
    Velocity Speed=&RandomRangeFloat(Minimum=300, Maximum=470)
   )
   &SubUVAnimation(
    Evaluation Type=::NewEnumerator0
    "Sprite Renderer "=DataInterface() {propertyValues {SpriteRenderer=&self:SpawnBurst_0.NiagaraSpriteRendererProperties_0}}
    SubUV Animation Mode:ENiagara_SubUVLookupModeV2=::NewEnumerator2
    UseEndFrame=false
    UseStartFrame=false
   )
   /SetVariables(Particles.VisibilityTag=&SpawnGroup()) {isSetParameters=true}
   /SetVariables(Particles.Glow Size=&Vector2DFromFloat(Value=&RandomRangeFloat(Minimum=4, Maximum=8))) {isSetParameters=true}
  ]
  particleUpdateScript [
   &ParticleState()
   &ScaleSpriteSize(Uniform Curve Sprite Scale=DataInterface() {propertyValues {Curve {keys [
        <at 0 0 RCIM_Cubic>
        {arriveTangentWeight=0.1067, interpMode=RCIM_Cubic, leaveTangentWeight=0.3202, tangentMode=RCTM_User, time=0.2, value=1}
        {arriveTangentWeight=0.3202, interpMode=RCIM_Cubic, leaveTangentWeight=0.1067, tangentMode=RCTM_User, time=0.8, value=1}
        <at 1 0 RCIM_Cubic>
       ]}}})
   &ScaleSpriteSize(Uniform Curve Sprite Scale=DataInterface() {propertyValues {Curve {keys [
        <at 0 0 RCIM_Cubic>
        {arriveTangentWeight=0.1067, interpMode=RCIM_Cubic, leaveTangentWeight=0.3202, tangentMode=RCTM_User, time=0.2, value=1}
        {arriveTangentWeight=0.3202, interpMode=RCIM_Cubic, leaveTangentWeight=0.1067, tangentMode=RCTM_User, time=0.8, value=1}
        <at 1 0 RCIM_Cubic>
       ]}}}) {module=ScaleSpriteSize001}
   &VortexForce(
    Origin Pull Amount=2542.4172
    UseOriginOffset=true
    Vortex Force Amount=&RandomRangeFloat(Minimum=-2000, Maximum=-3000)
    Vortex Origin Offset:Vector3f=&SubtractVectorFromPosition(Position=Linked() {linkedVariable {name=System.HeartIntroSpawnPosition, type=Vector3f}}, Vector=<0 0 50>)
   )
   &Drag(
    Drag=&RandomRangeFloat(Minimum=2.8, Maximum=3)
    Rotational Drag=5
    Use Linear Drag=true
    Use Rotational Drag=true
   )
   &SolveForcesAndVelocity(Clamp Velocity=false, Limit Acceleration=false)
   &ScaleColor(
    Linear Color Curve=DataInterface() {propertyValues {
      AlphaCurve {keys [<at 0 0>,<at 0.1218 0.7>,<at 0.803 0.7>,<at 1 0>]}
      BlueCurve {keys [<at 0 1>,<at 1 1>]}
      GreenCurve {keys [<at 0 1>,<at 1 1>]}
      RedCurve {keys [<at 0 1>,<at 1 1>]}
     }}
    Scale Mode=2
   )
  ]
  rendererClasses [&NiagaraSpriteRendererProperties]
  renderers [
   {rendererClass=&NiagaraSpriteRendererProperties}
   {rendererClass=&NiagaraSpriteRendererProperties, rendererIndex=1}
  ]
 }
 Trail {
  emitterUpdateScript [
   &EmitterState()
   &SpawnRate(
    SpawnGroup=1
    SpawnRate=&FloatFromCurve(
     FloatCurve=DataInterface() {propertyValues {Curve {keys [<at 0 1 RCIM_Constant>,<at 0.22 0 RCIM_Constant>]}}}
     CurveIndex=$Emitter.NormalizedLoopAge
     Scale Curve=100
    )
    Use Spawn Probability=false
   )
  ]
  particleSpawnScript [
   &PartitionParticles(Write Partition to Ribbon ID=::NewEnumerator0)
   &InitializeParticle(
    Color=&MultiplyLinearColor(Linear Color=<rgba 0.8125 0.8594 1 1>, Scale Factor=&MakeVector4_FromLinearColor(LinearColor=$System.TOD))
    Lifetime=0.7
    Ribbon Width=&Multiply_Float(A=5, B=&FN_FrontEndFloatModifier(Gameplay=1, Front End=$User.FrontEndScale))
    Ribbon Width Mode=::NewEnumerator1
    Sprite Rotation Mode=::NewEnumerator0
    UsePositionOffset=false
    Write Material Random=true
   )
   /SetVariables(Particles.Position=&Add_Vector(
     A=<0 0 6>
     B=&Add_Vector(
      A=&Lerp_Vector(
       A=$System.HeartIntroSpawnPosition
       B=$System.BurstPosition
       Alpha=&FloatFromCurve(
        FloatCurve=DataInterface() {propertyValues {Curve {keys [{arriveTangent=2.0116, interpMode=RCIM_Cubic, leaveTangent=2.0116, tangentMode=RCTM_User, time=0, value=0}, <at 1 1 RCIM_Cubic>]}}}
        CurveIndex=&Divide_Float(A=$Emitter.Age, B=$System.IntroLength)
       )
      )
      B=&MakeVector(
       X=&Waveform(
        Global Input Position=&MergeHelper_FloatSwitch(0=0, 1=1, 2=?, 3=?, Index=$Particles.PartitionParticles.Partition, Num=::NewEnumerator0)
        Global Phase=&Multiply_Float(A=$Emitter.Age, B=1.2)
        "Waveform [1]"=::NewEnumerator0
        "[1] Frequency"=1.5
        "[1] Amplitude Min/Max"=<-1 1>
        "[1] Scale"=&FloatFromCurve(
         FloatCurve=DataInterface() {propertyValues {Curve {keys [
             {arriveTangent=5.9356, interpMode=RCIM_Cubic, leaveTangent=5.9356, tangentMode=RCTM_User, time=-0.2504, value=-0.4466}
             <at 0.2496 0.5534 RCIM_Cubic>
            ]}}}
         CurveIndex=$Emitter.Age
         Scale Curve=&Multiply_Float(A=10, B=&FN_FrontEndFloatModifier(Gameplay=1, Front End=$User.FrontEndScale))
        )
       )
       Y=&Waveform(
        Global Input Position=&MergeHelper_FloatSwitch(0=0, 1=1, 2=?, 3=?, Index=$Particles.PartitionParticles.Partition, Num=::NewEnumerator0)
        Global Phase=&Multiply_Float(A=$Emitter.Age, B=1.2)
        "Waveform [1]"=::NewEnumerator1
        "[1] Frequency"=1.5
        "[1] Amplitude Min/Max"=<-1 1>
        "[1] Scale"=&FloatFromCurve(
         FloatCurve=DataInterface() {propertyValues {Curve {keys [
             {arriveTangent=5.9356, interpMode=RCIM_Cubic, leaveTangent=5.9356, tangentMode=RCTM_User, time=-0.2504, value=-0.4466}
             <at 0.2496 0.5534 RCIM_Cubic>
            ]}}}
         CurveIndex=$Emitter.Age
         Scale Curve=&Multiply_Float(A=10, B=&FN_FrontEndFloatModifier(Gameplay=1, Front End=$User.FrontEndScale))
        )
       )
       Z=0
      )
     )
    )) {isSetParameters=true}
   /SetVariables(Particles.RibbonUVDistance:NiagaraFloat=&DistanceBetweenPositions(
     A=&Lerp_Vector(
      A=$System.HeartIntroSpawnPosition
      B=$System.BurstPosition
      Alpha=&FloatFromCurve(
       FloatCurve=DataInterface() {propertyValues {Curve {keys [
           {arriveTangent=0.7913, interpMode=RCIM_Cubic, leaveTangent=0.7913, tangentMode=RCTM_User, time=0, value=0}
           {arriveTangent=-0.1073, tangentMode=RCTM_User, time=1, value=1}
          ]}}}
       CurveIndex=&Divide_Float(A=$Emitter.Age, B=$System.IntroLength)
      )
     )
     B=<0 0 0>
    )) {isSetParameters=true}
   /SetVariables(Particles.SpawnAge:NiagaraFloat=&Divide_Float(A=$Emitter.Age, B=$System.IntroLength)) {isSetParameters=true}
  ]
  particleUpdateScript [
   &ParticleState()
   &ScaleColor(
    Scale Alpha=&Multiply_Float(
     A=&FloatFromCurve(FloatCurve=DataInterface() {propertyValues {Curve {
         defaultValue=0
         keys [
          {arriveTangent=-0.2441, time=0.5, value=1}
          {arriveTangent=-0.0405, interpMode=RCIM_Cubic, leaveTangent=-0.0405, tangentMode=RCTM_User, time=1, value=0}
         ]
        }}})
     B=1.2
    )
    Scale RGB=&MakeVectorFromLinearColorRGB(Color=&ColorFromCurve(ColorCurve=DataInterface() {propertyValues {
        AlphaCurve {keys [<at 0 1>]}
        BlueCurve {keys [<at 0 0.9688>,<at 0.2975 0.3669>,<at 0.9945 0.1492>]}
        GreenCurve {keys [<at 0 0.7035>,<at 0.2975 0.3025>,<at 0.9945 0.0736>]}
        RedCurve {keys [<at 0 0.2294>,<at 0.2975 0.1688>,<at 0.9945 0.0273>]}
       }}))
    ScaleA=true
    ScaleRGB=true
   )
   &DynamicMaterialParameters(
    Index 0 Param 1=-0.5
    Index 0 Param 3=1
    Index 0 Param 4=1
    Module.Index0Float0Write=true
    Module.Index0Float1Write=true
    Module.Index0Float2Write=true
    Module.Index0Float3Write=true
   ) {module=DynamicMaterialParameters001}
   &ScaleRibbonWidth(Ribbon Width Scale=&Multiply_Float(
     A=&FloatFromCurve(FloatCurve=DataInterface() {propertyValues {Curve {keys [<at 0 0>,<at 0.1 1>,<at 1 2>]}}})
     B=&FloatFromCurve(FloatCurve=DataInterface() {propertyValues {Curve {keys [<at 0.05 0>,<at 0.15 1>,<at 0.8 1>,<at 0.9 0>]}}}, CurveIndex=$Particles.SpawnAge)
    ))
   &Drag(Drag=2, Use Linear Drag=true, Use Rotational Drag=true)
   &SolveForcesAndVelocity(Clamp Velocity=false, Limit Acceleration=false)
  ]
  rendererClasses [&NiagaraRibbonRendererProperties]
  renderers [{rendererClass=&NiagaraRibbonRendererProperties}]
 }
}
summary {
 systemName=NS_ReactFX_BriskImp
 emitters [
  {emitterName=Heart, rendererClasses [&NiagaraSpriteRendererProperties]}
  {emitterName=HeartGlow, rendererClasses [&NiagaraSpriteRendererProperties]}
  {emitterName=Trail, rendererClasses [&NiagaraRibbonRendererProperties]}
  {emitterName=LWE_AirGlow, rendererClasses []}
  {emitterName=LWE_Corona, rendererClasses []}
  {emitterName=Frosty_Smoke, rendererClasses [&NiagaraSpriteRendererProperties]}
  {emitterName=LWE_Rays001, rendererClasses []}
  {emitterName=LWE_Rays, rendererClasses []}
  {emitterName=LWE_Mist, rendererClasses []}
  {emitterName=Ice_Frags, rendererClasses [&NiagaraMeshRendererProperties]}
  {emitterName=SpawnBurst, rendererClasses [&NiagaraSpriteRendererProperties]}
  {emitterName=Ice_Frags001, rendererClasses [&NiagaraSpriteRendererProperties]}
 ]
 userVariables [
  User.bIsFrontEnd {type=NiagaraBool}
  User.bIsFrontEndPreview {type=NiagaraBool}
  User.FrontEndScale {type=NiagaraFloat, defaultValue=0.7, description=""}
  User.ParaGlideLeanAlpha {type=NiagaraFloat, defaultValue=0, description=""}
 ]
}
