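# Niagara reaction effect for cosmetic companions that plays a pet-interaction sound and a sprite intro ending in a rip-and-confetti burst.
# Intro sprites move from HeartIntroSpawnPosition to BurstPosition with curve-driven scale pops.
# Just before IntroLength ends, three GPU Ripping_Bits sprites split along camera-relative directions using a shared curve, while tricolor Confetti bursts outward under drag and gravity.
# User flags mark front-end contexts; three LWE glow emitters list no renderers.
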NiagaraSystem /CosmeticCompanions/FX/React/NS_ReactFX_BuffCut
dependencies {
 usedDataInterfaces [
  &NiagaraDataInterfaceAudioPlayer
  &NiagaraDataInterfaceCurve
  &NiagaraDataInterfaceCamera
  &NiagaraDataInterfaceArrayColor
 ]
 usedDynamicInputs [
  &SimulationPosition
  &FN_FrontEndFloatModifier
  /Niagara/DynamicInputs/TypeConversions/MakeCustomPositionFromBool
  &FN_FrontEndVector3Modifier
 ]
 usedModules [
  &self:SystemSpawnScript.SystemScriptSource.SystemScriptGraph.NiagaraNodeAssignment_3.SetVariables_#1
  &self:SystemSpawnScript.SystemScriptSource.SystemScriptGraph.NiagaraNodeAssignment_0.SetVariables_#2
  &self:SystemSpawnScript.SystemScriptSource.SystemScriptGraph.NiagaraNodeAssignment_1.SetVariables_#3
  &self:SystemSpawnScript.SystemScriptSource.SystemScriptGraph.NiagaraNodeAssignment_7.SetVariables_#4
  &self:SystemSpawnScript.SystemScriptSource.SystemScriptGraph.NiagaraNodeAssignment_4.SetVariables_#5
  &self:SystemSpawnScript.SystemScriptSource.SystemScriptGraph.NiagaraNodeAssignment_8.SetVariables_#6
  &self:SystemSpawnScript.SystemScriptSource.SystemScriptGraph.NiagaraNodeAssignment_9.SetVariables_#7
  &self:SystemSpawnScript.SystemScriptSource.SystemScriptGraph.NiagaraNodeAssignment_10.SetVariables_#8
  &self:SystemSpawnScript.SystemScriptSource.SystemScriptGraph.NiagaraNodeAssignment_5.SetVariables_#9
  &self:SystemSpawnScript.SystemScriptSource.SystemScriptGraph.NiagaraNodeAssignment_6.SetVariables_#10
  &SystemState
  &self:SystemSpawnScript.SystemScriptSource.SystemScriptGraph.NiagaraNodeAssignment_2.SetVariables_#11
 ]
}
emitters {
 Confetti {
  emitterUpdateScript [
   &EmitterState()
   &SpawnBurst_Instantaneous(Spawn Count=9, Spawn Time=&Subtract_Float(A=$System.IntroLength, B=0.05), Use Spawn Probability=false)
  ]
  particleSpawnScript [
   &InitializeParticle(
    Color=&SelectLinearColorFromArray(
     Array Sampling Mode=::NewEnumerator0
     Direct Array Index=?
     Color Selection Array=DataInterface() {propertyValues {ColorData [<rgba 1 0.3005 0.0723 1>, <rgba 1 1 1 1>, <rgba 0.2159 0.1779 0.7305 1>]}}
     Random Seed=0
    )
    Lifetime Max=1
    Lifetime Min=0.2
    Lifetime Mode=::NewEnumerator1
    Position Offset:NiagaraPosition=$System.BurstPosition
    Position Offset Coordinate Space=::Local
    Sprite Rotation Mode=::NewEnumerator2
    Sprite Size Mode=::NewEnumerator1
    Uniform Sprite Size Max=7
    Uniform Sprite Size Min=2
    UsePositionOffset=true
    Write Material Random=true
   ) {module=InitializeParticle001}
   &AddVelocity(
    Module.Use Falloff Distance=false
    Origin Offset:NiagaraPosition=$System.BurstPosition
    Override Seed=false
    Velocity Mode=::NewEnumerator1
    Velocity Speed=&RandomRangeFloat(Minimum=100, Maximum=250)
   ) {module=AddVelocity001}
  ]
  particleUpdateScript [
   &ParticleState()
   &Drag(Drag=&RandomRangeFloat(Minimum=5, Maximum=10), Use Linear Drag=true, Use Rotational Drag=false)
   &GravityForce(Gravity=&RandomRangeVector(Minimum=<0 0 -360>, Maximum=<1 1 -200>))
   &SolveForcesAndVelocity(Clamp Velocity=false, Limit Acceleration=false)
   &ScaleSpriteSize(Scale Factor=&Vector2DFromFloat(Value=&FloatFromCurve(FloatCurve=DataInterface() {propertyValues {Curve {keys [
          {arriveTangent=-0.4596, interpMode=RCIM_Cubic, leaveTangent=-0.4596, tangentMode=RCTM_User, time=0.55, value=1}
          {arriveTangent=-4.9031, interpMode=RCIM_Cubic, leaveTangent=-4.9031, tangentMode=RCTM_User, time=1, value=0}
         ]}}})))
  ]
  rendererClasses [&NiagaraSpriteRendererProperties]
  renderers [{rendererClass=&NiagaraSpriteRendererProperties}]
 }
 Intro {
  emitterSpawnScript [
   &InitializeAudioPlayer(Audio Player=DataInterface() {propertyValues {
      Attenuation=null
      Concurrency=null
      MaxPlaysPerTick=10
      SoundToPlay=/CosmeticCompanions/Assets/Audio_Shared/Interaction/Petting/SC_Mimosa_Interaction_React_Default
      bAllowLoopingOneShotSounds=false
      bLimitPlaysPerTick=true
      bOnlyActiveDuringGameplay=false
      bStopWhenComponentIsDestroyed=true
      ConfigurationUserParameter {parameter {name=None, typeDefHandle {registeredTypeIndex=1}, varData []}}
      ParameterNames []
     }})
   /Niagara/Modules/Audio/PlayAudio() {module=PlayAudio001}
  ]
  emitterUpdateScript [
   &EmitterState()
   &SpawnBurst_Instantaneous(Use Spawn Probability=false, UseLoopCountLimit=false)
  ]
  particleSpawnScript [
   &PartitionParticles()
   &InitializeParticle(
    Lifetime=$System.IntroLength
    Sprite Rotation Mode=::NewEnumerator0
    Sprite Size Mode=::NewEnumerator3
    Uniform Sprite Size=8
    UsePositionOffset=false
    Write Material Random=true
   )
  ]
  particleUpdateScript [
   &ParticleState()
   &ScaleSpriteSize(
    Scale Sprite Size Mode=::NewEnumerator0
    Uniform Scale Factor=&Add_Float(
     A=&FloatFromCurve(FloatCurve=DataInterface() {propertyValues {Curve {
         defaultValue=0.5
         keys [
          {arriveTangent=-0.5031, interpMode=RCIM_Cubic, leaveTangent=-0.5031, tangentMode=RCTM_User, time=0.8, value=0.5}
          {arriveTangent=-0.0761, interpMode=RCIM_Cubic, leaveTangent=-0.0761, time=0.9056, value=0.0134}
          {arriveTangent=90.1137, interpMode=RCIM_Cubic, leaveTangent=90.1137, tangentMode=RCTM_User, time=0.9314, value=0.49}
          {arriveTangent=-49.4177, interpMode=RCIM_Cubic, leaveTangent=-49.4177, tangentMode=RCTM_User, time=0.9444, value=0.8639}
         ]
        }}})
     B=&MergeHelper_FloatSwitch(
      0=&FloatFromCurve(FloatCurve=DataInterface() {propertyValues {Curve {
          defaultValue=0
          keys [
           {arriveTangent=3.2329, interpMode=RCIM_Cubic, leaveTangent=3.2329, tangentMode=RCTM_User, time=0, value=0}
           {arriveTangent=0.6241, interpMode=RCIM_Cubic, leaveTangent=0.6241, tangentMode=RCTM_User, time=0.1569, value=0.2299}
           {arriveTangent=2.2355, interpMode=RCIM_Cubic, leaveTangent=2.2355, tangentMode=RCTM_User, time=0.4075, value=0.6225}
           {arriveTangent=3.5453, interpMode=RCIM_Cubic, leaveTangent=3.5453, time=0.5461, value=1.084}
           {arriveTangent=4.4217, interpMode=RCIM_Cubic, leaveTangent=4.4217, time=0.6608, value=1.5207}
           {arriveTangent=-0.9053, interpMode=RCIM_Cubic, leaveTangent=-0.9053, tangentMode=RCTM_User, time=0.7123, value=1.819}
           {arriveTangent=-2.847, interpMode=RCIM_Cubic, leaveTangent=-2.847, time=0.7406, value=1.9236}
           {arriveTangent=-0.1202, interpMode=RCIM_Cubic, leaveTangent=-0.1202, tangentMode=RCTM_User, time=1, value=1}
          ]
         }}})
      1=&FloatFromCurve(
       FloatCurve=DataInterface() {propertyValues {Curve {
          defaultValue=0
          keys [
           {arriveTangent=0.087, interpMode=RCIM_Cubic, leaveTangent=0.087, tangentMode=RCTM_User, time=0, value=0}
           {arriveTangent=1.0515, interpMode=RCIM_Cubic, leaveTangent=1.0515, time=0.3888, value=0.1798}
           {arriveTangent=2.4243, interpMode=RCIM_Cubic, leaveTangent=2.4243, tangentMode=RCTM_User, time=0.5921, value=0.6225}
           {arriveTangent=1.5149, interpMode=RCIM_Cubic, leaveTangent=1.5149, tangentMode=RCTM_User, time=0.701, value=1.2762}
           {arriveTangent=-0.1202, interpMode=RCIM_Cubic, leaveTangent=-0.1202, tangentMode=RCTM_User, time=1, value=0.7243}
          ]
         }}}
       Scale Curve=0.7
      )
      2=?
      3=?
      Index=$Particles.PartitionParticles.Partition
      Num=::NewEnumerator0
     )
    )
   ) {module=ScaleSpriteSize002}
   &ScaleColor(
    Scale Alpha=&FloatFromCurve(FloatCurve=DataInterface() {propertyValues {Curve {
        defaultValue=0
        keys [
         {arriveTangent=9.0725, interpMode=RCIM_Cubic, leaveTangent=9.0725, tangentMode=RCTM_User, time=0, value=0}
         {arriveTangent=-0.2441, leaveTangent=-0.2441, time=0.1408, value=1}
         {arriveTangent=-1.146, leaveTangent=-1.146, time=0.97, value=1}
         {arriveTangent=-0.0405, interpMode=RCIM_Cubic, leaveTangent=-0.0405, tangentMode=RCTM_User, time=0.971, value=0}
        ]
       }}})
    ScaleA=true
    ScaleRGB=true
   )
   &OffsetPosition(Position Offset=&Lerp_Vector(
     A=$System.HeartIntroSpawnPosition
     B=$System.BurstPosition
     Alpha=&FloatFromCurve(FloatCurve=DataInterface() {propertyValues {Curve {keys [
          {arriveTangent=0.7913, interpMode=RCIM_Cubic, leaveTangent=0.7913, tangentMode=RCTM_User, time=0, value=0}
          {arriveTangent=-0.1074, tangentMode=RCTM_User, time=1, value=1}
         ]}}})
    ))
   &DynamicMaterialParameters(
    Index 0 Param 2=1
    Module.Index0Float0Write=true
    Module.Index0Float1Write=true
    Module.Index0Float2Write=true
    Module.Index0Float3Write=true
   )
  ]
  rendererClasses [&NiagaraSpriteRendererProperties]
  renderers [{rendererClass=&NiagaraSpriteRendererProperties}]
 }
 LWE_AirGlow {rendererClasses [], renderers []}
 LWE_Corona {rendererClasses [], renderers []}
 LWE_Rays {rendererClasses [], renderers []}
 Ripping_Bits {
  simTarget=GPUComputeSim
  emitterUpdateScript [
   &EmitterState()
   &SpawnBurst_Instantaneous(
    Spawn Count=3
    Spawn Time=&Subtract_Float(A=$System.IntroLength, B=0.07)
    Use Spawn Probability=false
    UseLoopCountLimit=false
   )
  ]
  particleSpawnScript [
   &InitializeParticle(
    Lifetime=$System.IntroLength
    Sprite Rotation Mode=::NewEnumerator0
    Sprite Size Mode=::NewEnumerator3
    UsePositionOffset=true
    Write Material Random=true
   )
   /SetVariables(Particles.SubImageIndex=&MakeFloatFromInt(Int=&Add_Integer(A=&ReturnExecIndex(), B=1))) {isSetParameters=true}
   &DynamicMaterialParameters(
    Index 0 Param 3=1
    Module.Index0Float0Write=true
    Module.Index0Float1Write=true
    Module.Index0Float2Write=true
    Module.Index0Float3Write=true
   )
  ]
  particleUpdateScript [
   &ParticleState()
   &SpriteRotationRate(Rotation Rate=&FloatFromCurve(
     FloatCurve=DataInterface() {propertyValues {Curve {
        defaultValue=1
        keys [
         {arriveTangent=2.6847, interpMode=RCIM_Cubic, leaveTangent=2.6847, tangentMode=RCTM_User, time=0, value=0}
         {arriveTangent=0.1978, interpMode=RCIM_Cubic, leaveTangent=0.1978, tangentMode=RCTM_User, time=1.0073, value=1.0194}
        ]
       }}}
     Scale Curve=&UniformAOrBFloat(A=-120, B=120, Evaluation Type=::NewEnumerator0, Recalculate Random Each Loop=true, Override Seed=false)
    ))
   &ScaleSpriteSize(
    Scale Sprite Size Mode=::NewEnumerator0
    Uniform Scale Factor=&FloatFromCurve(FloatCurve=DataInterface() {propertyValues {Curve {
        defaultValue=0
        keys [
         {arriveTangent=18.0926, interpMode=RCIM_Cubic, leaveTangent=18.0926, tangentMode=RCTM_User, time=0, value=1}
         {arriveTangent=-1, interpMode=RCIM_Cubic, leaveTangent=-1, time=0.1617, value=1.073}
         {arriveTangent=-0.0405, interpMode=RCIM_Cubic, leaveTangent=-0.0405, tangentMode=RCTM_User, time=1, value=0}
        ]
       }}})
   ) {module=ScaleSpriteSize002}
   &ScaleColor(
    Scale Alpha=&FloatFromCurve(FloatCurve=DataInterface() {propertyValues {Curve {
        defaultValue=0
        keys [
         {arriveTangent=9.0725, interpMode=RCIM_Cubic, leaveTangent=9.0725, tangentMode=RCTM_User, time=0.0425, value=0}
         {arriveTangent=-0.2441, time=0.05, value=1}
         {arriveTangent=-1.146, time=0.3022, value=0.3765}
         {arriveTangent=-0.0405, interpMode=RCIM_Cubic, leaveTangent=-0.0405, tangentMode=RCTM_User, time=1, value=0}
        ]
       }}})
    ScaleA=true
    ScaleRGB=true
   )
   &OffsetPosition(Position Offset=&AddVectorToPosition(
     Position=Linked() {linkedVariable {name=System.BurstPosition, type=Vector3f}}
     Vector=&MergeHelper_VectorSwitch(
      0=&Multiply_VectorByFloat(
       Vector=&NormalizeVector(
        Vector To Normalize=&Add_Vector(
         A=&Multiply_VectorByFloat(Vector=&ReturnCameraVector(Camera Vectors=::NewEnumerator3, Coordinate Space=::Local), Float=-9.2359)
         B=&Multiply_VectorByFloat(Vector=&ReturnCameraVector(Camera Vectors=::NewEnumerator2, Coordinate Space=::Local), Float=9.3222)
        )
        Safe Normalize=true
        Safe Fallback Threshold=0
        Safe Fallback Vector=<0 0 0.001>
       )
       Float=&FloatFromCurve(FloatCurve=$System.RippingCurveAnim, Scale Curve=20)
      )
      1=&Multiply_VectorByFloat(
       Vector=&NormalizeVector(
        Vector To Normalize=&Multiply_VectorByFloat(Vector=&ReturnCameraVector(Camera Vectors=::NewEnumerator2, Coordinate Space=::Local), Float=-5)
        Safe Normalize=true
        Safe Fallback Threshold=0
        Safe Fallback Vector=<0 0 0.001>
       )
       Float=&FloatFromCurve(FloatCurve=$System.RippingCurveAnim, Scale Curve=10)
      )
      2=&Multiply_VectorByFloat(
       Vector=&NormalizeVector(
        Vector To Normalize=&Add_Vector(
         A=&Multiply_VectorByFloat(Vector=&ReturnCameraVector(Camera Vectors=::NewEnumerator3, Coordinate Space=::Local), Float=19.6906)
         B=&Multiply_VectorByFloat(Vector=&ReturnCameraVector(Camera Vectors=::NewEnumerator2, Coordinate Space=::Local), Float=5)
        )
        Safe Normalize=true
        Safe Fallback Threshold=0
        Safe Fallback Vector=<0 0 0.001>
       )
       Float=&FloatFromCurve(FloatCurve=$System.RippingCurveAnim, Scale Curve=20)
      )
      3=?
      Index=&ReturnExecIndex()
      Num=::NewEnumerator1
     )
    )) {module=OffsetPosition001}
  ]
  rendererClasses [&NiagaraSpriteRendererProperties]
  renderers [{rendererClass=&NiagaraSpriteRendererProperties}]
 }
}
summary {
 systemName=NS_ReactFX_BuffCut
 emitters [
  {emitterName=Intro, rendererClasses [&NiagaraSpriteRendererProperties]}
  {emitterName=Ripping_Bits, simTarget=GPUComputeSim, rendererClasses [&NiagaraSpriteRendererProperties]}
  {emitterName=Confetti, rendererClasses [&NiagaraSpriteRendererProperties]}
  {emitterName=LWE_AirGlow, rendererClasses []}
  {emitterName=LWE_Rays, rendererClasses []}
  {emitterName=LWE_Corona, rendererClasses []}
 ]
 userVariables [
  User.bIsFrontEnd {type=NiagaraBool}
  User.bIsFrontEndPreview {type=NiagaraBool}
 ]
}
