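# Niagara reaction effect for the BurgleBoss cosmetic companion: an animated intro sprite that bursts into gold.
# Intro plays a shared petting-reaction sound and grows a glowing flipbook sprite at BurstPosition for the length of IntroLength.
# IntroLength times the burst: a multi-renderer sprite and mesh Pop, flying gold pieces, gold dust, a three-way dust spray and two brief flashes.
# Front-end user flags switch most sizes to a system-level FE Mult scale.
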NiagaraSystem /CosmeticCompanions/FX/React/NS_ReactFX_BurgleBoss
dependencies {
 usedDataInterfaces [
  &NiagaraDataInterfaceAudioPlayer
  &NiagaraDataInterfaceCurve
  &NiagaraDataInterfaceSpriteRendererInfo
  &NiagaraDataInterfaceVectorCurve
  &NiagaraDataInterfaceCamera
  &NiagaraDataInterfaceArrayColor
  &NiagaraDataInterfaceColorCurve
 ]
 usedDynamicInputs [&FN_FrontEndVector3Modifier]
 usedModules [
  &self:SystemSpawnScript.SystemScriptSource.SystemScriptGraph.NiagaraNodeAssignment_0.SetVariables_#1
  &self:SystemSpawnScript.SystemScriptSource.SystemScriptGraph.NiagaraNodeAssignment_1.SetVariables_#2
  &self:SystemSpawnScript.SystemScriptSource.SystemScriptGraph.NiagaraNodeAssignment_4.SetVariables_#3
  &self:SystemSpawnScript.SystemScriptSource.SystemScriptGraph.NiagaraNodeAssignment_2.SetVariables_#4
  &self:SystemSpawnScript.SystemScriptSource.SystemScriptGraph.NiagaraNodeAssignment_3.SetVariables_#5
  &self:SystemSpawnScript.SystemScriptSource.SystemScriptGraph.NiagaraNodeAssignment_9.SetVariables_#6
  &self:SystemSpawnScript.SystemScriptSource.SystemScriptGraph.NiagaraNodeAssignment_10.SetVariables_#7
  &self:SystemSpawnScript.SystemScriptSource.SystemScriptGraph.NiagaraNodeAssignment_5.SetVariables_#8
  &self:SystemSpawnScript.SystemScriptSource.SystemScriptGraph.NiagaraNodeAssignment_6.SetVariables_#9
  &self:SystemSpawnScript.SystemScriptSource.SystemScriptGraph.NiagaraNodeAssignment_7.SetVariables_#10
  &SystemState
 ]
}
emitters {
 GoldDust {
  emitterUpdateScript [
   &EmitterState(Life Cycle Mode=::NewEnumerator1, Loop Behavior=::NewEnumerator1, Loop Duration=3, UseLoopDelay=false)
   &SpawnBurst_Instantaneous(
    Spawn Count=3
    Spawn Time=&Subtract_Float(A=$System.IntroLength, B=0)
    Use Spawn Probability=false
    UseLoopCountLimit=false
   )
  ]
  particleSpawnScript [
   &InitializeParticle(
    Color=&SelectLinearColorFromArray(
     Array Sampling Mode=::NewEnumerator0
     Direct Array Index=?
     Color Selection Array=DataInterface() {propertyValues {ColorData [<rgba 8 4.0072 0.7917 0.1>, <rgba 8 5.6344 1.3957 0.1>, <rgba 8 2.2178 0.0178 0.1>, <rgba 8 3.6879 0.2005 0.1>, <rgba 8 4.0072 0.7917 0.1>]}}
     Random Seed=0
    )
    Lifetime=&UniformAOrBFloat(
     A=&RandomRangeFloat(Minimum=0.5, Maximum=0.8)
     B=&RandomRangeFloat(Minimum=1, Maximum=1.2)
     Distribution Weight=0.3
     Evaluation Type=::NewEnumerator0
     Recalculate Random Each Loop=true
     Override Seed=false
    )
    Position Offset=$System.BurstPosition
    Sprite Rotation Mode=::NewEnumerator2
    Sprite Size Mode=::NewEnumerator3
    Uniform Sprite Size=&Multiply_Float(
     A=&MakeCustomFloatFromBool(
      Bool=&Boolean_OrOperation(Bool1=$User.bIsFrontEnd, Bool2=$User.bIsFrontEndPreview)
      True Float=$System.FE Mult
      False Float=1
     )
     B=&Multiply_Float(
      A=&RandomRangeFloat(Minimum=22, Maximum=40)
      B=&RandomRangeFloat(Minimum=1, Maximum=2.5)
     )
    )
    UsePositionOffset=true
    Write Material Random=true
   )
   &CameraOffset(Camera Offset Amount=&RandomRangeFloat(Minimum=37, Maximum=50))
   &ShapeLocation(
    MaskSpawnGroup=false
    Override Seed=false
    Sphere Radius=&Multiply_Float(
     A=&MakeCustomFloatFromBool(
      Bool=&Boolean_OrOperation(Bool1=$User.bIsFrontEnd, Bool2=$User.bIsFrontEndPreview)
      True Float=$System.FE Mult
      False Float=1
     )
     B=1
    )
    UseOwnerScale=true
   )
   &ApplyInitialForces()
  ]
  particleUpdateScript [
   &ParticleState()
   &ScaleSpriteSize(Uniform Curve Sprite Scale=DataInterface() {propertyValues {Curve {
       defaultValue=0
       keys [
        {arriveTangent=1.1179, interpMode=RCIM_Cubic, leaveTangent=1.1179, tangentMode=RCTM_User, time=0, value=0.35}
        {arriveTangent=0.0224, interpMode=RCIM_Cubic, leaveTangent=0.0224, tangentMode=RCTM_User, time=0.7686, value=1.0094}
        {arriveTangent=1.402, interpMode=RCIM_Cubic, leaveTangent=1.402, tangentMode=RCTM_User, time=0.9978, value=1.46}
       ]
      }}})
   &ScaleColor(
    Scale Alpha=&FloatFromCurve(
     FloatCurve=DataInterface() {propertyValues {Curve {
        defaultValue=0
        keys [
         {arriveTangent=-3.0035, interpMode=RCIM_Cubic, leaveTangent=-3.0035, tangentMode=RCTM_User, time=0.0118, value=1.0048}
         {arriveTangent=-1.2935, interpMode=RCIM_Cubic, leaveTangent=-1.2935, tangentMode=RCTM_User, time=0.3271, value=0.3741}
         {arriveTangent=-0.0404, interpMode=RCIM_Cubic, leaveTangent=-0.0404, tangentMode=RCTM_User, time=1, value=0}
        ]
       }}}
     Scale Curve=&RandomRangeFloat(Minimum=2.5, Maximum=5)
    )
    Scale RGB=&MakeVectorFromLinearColorRGB(Color=&ColorFromCurve(ColorCurve=DataInterface() {propertyValues {
        AlphaCurve {keys []}
        BlueCurve {keys [<at 0.1264 3>, {interpMode=RCIM_Cubic, leaveTangent=-2.2895, time=0.2304, value=1}, <at 0.3261 1 RCIM_Cubic>]}
        GreenCurve {keys [<at 0.1264 3>, {interpMode=RCIM_Cubic, leaveTangent=-2.2895, time=0.2304, value=1}, <at 0.3261 1 RCIM_Cubic>]}
        RedCurve {keys [<at 0.1264 3>, {interpMode=RCIM_Cubic, leaveTangent=-2.2895, time=0.2304, value=1}, <at 0.3261 1 RCIM_Cubic>]}
       }}))
    ScaleA=true
    ScaleRGB=true
   )
   &Drag(Use Linear Drag=true, Use Rotational Drag=true)
   &GravityForce(Gravity=<0 0 -30>)
   &SolveForcesAndVelocity(Clamp Velocity=false, Limit Acceleration=false)
  ]
  rendererClasses [&NiagaraSpriteRendererProperties]
  renderers [{rendererClass=&NiagaraSpriteRendererProperties}]
 }
 GoldDust001 {
  emitterUpdateScript [
   &EmitterState(
    Life Cycle Mode=::NewEnumerator1
    Loop Behavior=::NewEnumerator1
    Loop Delay=&Subtract_Float(A=$System.IntroLength, B=0.05)
    Loop Duration=1.2
    UseLoopDelay=true
   )
   &SpawnRate(
    SpawnRate=&FloatFromCurve(
     FloatCurve=DataInterface() {propertyValues {Curve {keys [<at 0.0955 1.0094>,<at 0.2421 -0.0659>]}}}
     CurveIndex=$Emitter.NormalizedLoopAge
     Scale Curve=200
    )
    Use Spawn Probability=false
   )
   &SpawnBurst_Instantaneous(Spawn Count=7, Use Spawn Probability=false, UseLoopCountLimit=false)
  ]
  particleSpawnScript [
   &PartitionParticles(Number Of Partitions=3)
   &InitializeParticle(
    Color=&SelectLinearColorFromArray(
     Array Sampling Mode=::NewEnumerator0
     Direct Array Index=?
     Color Selection Array=DataInterface() {propertyValues {ColorData [<rgba 8 4.0072 0.7917 0.1>, <rgba 8 5.6344 1.3957 0.1>, <rgba 8 2.2178 0.0178 0.1>, <rgba 8 3.6879 0.2005 0.1>, <rgba 8 4.0072 0.7917 0.1>]}}
     Random Seed=0
    )
    Lifetime=&RandomRangeFloat(Minimum=0.3, Maximum=0.5)
    Position Offset=$System.BurstPosition
    Sprite Rotation Mode=::NewEnumerator2
    Sprite Size Mode=::NewEnumerator3
    Uniform Sprite Size=&Multiply_Float(
     A=&MakeCustomFloatFromBool(
      Bool=&Boolean_OrOperation(Bool1=$User.bIsFrontEnd, Bool2=$User.bIsFrontEndPreview)
      True Float=$System.FE Mult
      False Float=1
     )
     B=&RandomRangeFloat(Minimum=14, Maximum=8)
    )
    UsePositionOffset=true
    Write Material Random=true
   )
   &CameraOffset(Camera Offset Amount=&RandomRangeFloat(Minimum=37, Maximum=50))
   &ShapeLocation(
    MaskSpawnGroup=false
    Offset=&MergeHelper_VectorSwitch(0=<-5 0 4>, 1=<0 0 1>, 2=<7 0 8>, 3=?, Index=$Particles.PartitionParticles.Partition, Num=::NewEnumerator1)
    Offset Mode=::NewEnumerator0
    Override Seed=false
    Sphere Radius=&Multiply_Float(
     A=&MakeCustomFloatFromBool(
      Bool=&Boolean_OrOperation(Bool1=$User.bIsFrontEnd, Bool2=$User.bIsFrontEndPreview)
      True Float=$System.FE Mult
      False Float=1
     )
     B=1
    )
    UseOwnerScale=true
   )
   &ApplyInitialForces()
   &AddVelocity(
    Cone Axis=&MergeHelper_VectorSwitch(0=<-2 0 1>, 1=<1 0 1>, 2=<3 0 1>, 3=?, Index=$Particles.PartitionParticles.Partition, Num=::NewEnumerator1)
    Module.UseConeFalloff=false
    Module.UseVelDistribution=false
    Override Seed=false
    Velocity Mode=::NewEnumerator2
    Velocity Speed=&RandomRangeFloat(Minimum=100, Maximum=110)
   )
  ]
  particleUpdateScript [
   &ParticleState()
   &ScaleSpriteSize(Uniform Curve Sprite Scale=DataInterface() {propertyValues {Curve {
       defaultValue=0
       keys [
        {arriveTangent=1.1179, interpMode=RCIM_Cubic, leaveTangent=1.1179, tangentMode=RCTM_User, time=0, value=0.35}
        {arriveTangent=0.0224, interpMode=RCIM_Cubic, leaveTangent=0.0224, tangentMode=RCTM_User, time=0.7686, value=1.0094}
        {arriveTangent=1.402, interpMode=RCIM_Cubic, leaveTangent=1.402, tangentMode=RCTM_User, time=0.9978, value=1.46}
       ]
      }}})
   &ScaleColor(
    Scale Alpha=&Multiply_Float(
     A=&FloatFromCurve(
      FloatCurve=DataInterface() {propertyValues {Curve {
         defaultValue=0
         keys [
          {arriveTangent=9.0725, interpMode=RCIM_Cubic, leaveTangent=9.0725, tangentMode=RCTM_User, time=0, value=0}
          {arriveTangent=-2.5249, interpMode=RCIM_Cubic, leaveTangent=-2.5249, tangentMode=RCTM_User, time=0.1408, value=0.9847}
          {arriveTangent=-0.0405, interpMode=RCIM_Cubic, leaveTangent=-0.0405, tangentMode=RCTM_User, time=1, value=0}
         ]
        }}}
      Scale Curve=&RandomRangeFloat(Minimum=1, Maximum=2)
     )
     B=&FloatFromCurve(
      FloatCurve=DataInterface() {propertyValues {Curve {
         defaultValue=1
         keys [
          {interpMode=RCIM_Cubic, tangentMode=RCTM_User, time=0, value=0}
          {
           arriveTangent=2.4854
           arriveTangentWeight=0.0958
           interpMode=RCIM_Cubic
           leaveTangent=2.4854
           leaveTangentWeight=0.1505
           tangentMode=RCTM_User
           time=0.2186
           value=0.8111
          }
          {arriveTangentWeight=0.2594, interpMode=RCIM_Cubic, leaveTangentWeight=0.1499, tangentMode=RCTM_User, time=0.5146, value=1.0343}
          {arriveTangent=-2.1309, interpMode=RCIM_Cubic, leaveTangent=-2.1309, time=0.769, value=0.3455}
          {tangentMode=RCTM_User, time=1, value=0}
         ]
        }}}
      CurveIndex=$Emitter.NormalizedLoopAge
     )
    )
    Scale RGB=&MakeVectorFromLinearColorRGB(Color=&ColorFromCurve(ColorCurve=DataInterface() {propertyValues {
        AlphaCurve {keys []}
        BlueCurve {keys [<at 0.1264 3>, {interpMode=RCIM_Cubic, leaveTangent=-2.2895, time=0.2304, value=1}, <at 0.3261 1 RCIM_Cubic>]}
        GreenCurve {keys [<at 0.1264 3>, {interpMode=RCIM_Cubic, leaveTangent=-2.2895, time=0.2304, value=1}, <at 0.3261 1 RCIM_Cubic>]}
        RedCurve {keys [<at 0.1264 3>, {interpMode=RCIM_Cubic, leaveTangent=-2.2895, time=0.2304, value=1}, <at 0.3261 1 RCIM_Cubic>]}
       }}))
    ScaleA=true
    ScaleRGB=true
   )
   &Drag(Use Linear Drag=true, Use Rotational Drag=true)
   &GravityForce(Gravity=<0 0 -800>)
   &Drag(Drag=3, Use Linear Drag=true, Use Rotational Drag=true) {module=Drag001}
   &SolveForcesAndVelocity(Clamp Velocity=false, Limit Acceleration=false)
  ]
  rendererClasses [&NiagaraSpriteRendererProperties]
  renderers [{rendererClass=&NiagaraSpriteRendererProperties}]
 }
 Gold_Pieces {
  emitterUpdateScript [
   &EmitterState()
   &SpawnBurst_Instantaneous(
    Spawn Count=7
    Spawn Time=&Subtract_Float(A=$System.IntroLength, B=0)
    Use Spawn Probability=false
    UseLoopCountLimit=false
   )
  ]
  particleSpawnScript [
   &PartitionParticles(Number Of Partitions=7)
   &InitializeParticle(
    Color=&FNTimeOfDayColor(
     Night Color=<rgba 0.4314 0.5862 0.8073 1>
     Morning Color=<rgba 0.7055 0.6772 0.824 1>
     Day Color=<rgba 0.6955 0.7947 0.9365 1>
     Evening Color=<rgba 0.6472 0.6447 0.824 1>
    )
    Lifetime=&UniformAOrBFloat(
     A=&RandomRangeFloat(Minimum=0.25, Maximum=0.4)
     B=&RandomRangeFloat(Minimum=0.5, Maximum=0.7)
     Distribution Weight=0.3
     Evaluation Type=::NewEnumerator0
     Recalculate Random Each Loop=true
     Override Seed=false
    )
    Position Offset=$System.BurstPosition
    Sprite Size Mode=::NewEnumerator3
    Uniform Sprite Size=&Multiply_Float(
     A=&MakeCustomFloatFromBool(
      Bool=&Boolean_OrOperation(Bool1=$User.bIsFrontEnd, Bool2=$User.bIsFrontEndPreview)
      True Float=$System.FE Mult
      False Float=1
     )
     B=&UniformAOrBFloat(
      A=&RandomRangeFloat(Minimum=11, Maximum=15)
      B=&RandomRangeFloat(Minimum=16, Maximum=19)
      Distribution Weight=0.3
      Evaluation Type=::NewEnumerator0
      Recalculate Random Each Loop=true
      Override Seed=false
     )
    )
    UsePositionOffset=true
    Write Material Random=true
   )
   &AddVelocity(
    Module.Use Falloff Distance=false
    Origin Offset=$System.BurstPosition
    Override Seed=false
    Velocity Mode=::NewEnumerator1
    Velocity Speed=&RandomRangeFloat(Minimum=200, Maximum=300)
   )
   &ApplyInitialForces()
   &ShapeLocation(
    Additional Quaternion Rotation=true
    Additional Yaw / Pitch / Roll=<0 0.25 0>
    Handle Radius:NiagaraFloat=&Multiply_Float(
     A=&MakeCustomFloatFromBool(
      Bool=&Boolean_OrOperation(Bool1=$User.bIsFrontEnd, Bool2=$User.bIsFrontEndPreview)
      True Float=$System.FE Mult
      False Float=1
     )
     B=2
    )
    Large Radius=&Multiply_Float(
     A=&MakeCustomFloatFromBool(
      Bool=&Boolean_OrOperation(Bool1=$User.bIsFrontEnd, Bool2=$User.bIsFrontEndPreview)
      True Float=$System.FE Mult
      False Float=1
     )
     B=10
    )
    MaskSpawnGroup=false
    Override Seed=false
    Rotation Mode=::NewEnumerator3
    Rotation Quaternion=&OrientMeshToVector(
     Look At Direction=&ReturnCameraVector(Camera Vectors=::NewEnumerator1, Coordinate Space=::Simulation)
     Reference Axis Vector=<1 0 0>
    )
    Shape Primitive=::NewEnumerator3
    UseOwnerScale=true
   )
  ]
  particleUpdateScript [
   &ParticleState()
   &ScaleSpriteSize(Uniform Curve Sprite Scale=DataInterface() {propertyValues {Curve {
       defaultValue=0
       keys [{interpMode=RCIM_Cubic, tangentMode=RCTM_User, time=0.4305, value=1}, {interpMode=RCIM_Cubic, tangentMode=RCTM_User, time=1, value=0}]
      }}})
   &SubUVAnimation(
    Direct Frame Index:NiagaraFloat=&MergeHelper_FloatSwitch(0=0, 1=1, 2=1, 3=3, 4=4, 5=5, 6=3, Index=$Particles.PartitionParticles.Partition, Num=::NewEnumerator5)
    "Sprite Renderer "=DataInterface() {propertyValues {SpriteRenderer=&self:Gold_Pieces_0.NiagaraSpriteRendererProperties_0}}
    SubUV Animation Mode:ENiagara_SubUVLookupModeV2=::NewEnumerator5
   )
   &Drag(Drag=&RandomRangeFloat(Minimum=6, Maximum=8), Use Linear Drag=true, Use Rotational Drag=true)
   &GravityForce(Gravity=&Multiply_VectorByFloat(
     Vector=<0 0 -300>
     Float=&FloatFromCurve(FloatCurve=DataInterface() {propertyValues {Curve {defaultValue=1, keys [<at 0.067 0 RCIM_Cubic>,<at 0.2478 1 RCIM_Cubic>]}}})
    ))
   &SolveForcesAndVelocity(Clamp Velocity=false, Limit Acceleration=false)
   &ScaleColor(Scale Alpha=5, ScaleA=true, ScaleRGB=true)
  ]
  rendererClasses [&NiagaraSpriteRendererProperties]
  renderers [{rendererClass=&NiagaraSpriteRendererProperties}]
 }
 Intro {
  emitterSpawnScript [
   &InitializeAudioPlayer(Audio Player=DataInterface() {propertyValues {
      Attenuation=null
      Concurrency=null
      MaxPlaysPerTick=10
      SoundToPlay=/CosmeticCompanions/Assets/Audio_Shared/Interaction/Petting/SC_Mimosa_Interaction_React_Default
      bAllowLoopingOneShotSounds=false
      bLimitPlaysPerTick=true
      bOnlyActiveDuringGameplay=false
      bStopWhenComponentIsDestroyed=true
      ConfigurationUserParameter {parameter {name=None, typeDefHandle {registeredTypeIndex=1}, varData []}}
      ParameterNames []
     }})
   /Niagara/Modules/Audio/PlayAudio() {module=PlayAudio001}
  ]
  emitterUpdateScript [
   &EmitterState()
   &SpawnBurst_Instantaneous(Use Spawn Probability=false, UseLoopCountLimit=false)
  ]
  particleSpawnScript [&InitializeParticle(
    Color=&FNTimeOfDayColor(
     Night Color=<rgba 0.4314 0.5862 0.8073 1>
     Morning Color=<rgba 0.7055 0.6772 0.824 1>
     Day Color=<rgba 0.6955 0.7947 0.9365 1>
     Evening Color=<rgba 0.6472 0.6447 0.824 1>
    )
    Lifetime=$System.IntroLength
    Sprite Rotation Mode=::NewEnumerator0
    Sprite Size Mode=::NewEnumerator3
    Uniform Sprite Size=&MakeCustomFloatFromBool(
     Bool=&Boolean_OrOperation(Bool1=$User.bIsFrontEnd, Bool2=$User.bIsFrontEndPreview)
     True Float=&Multiply_Float(A=15, B=$System.FE Mult)
     False Float=15
    )
    UsePositionOffset=false
    Write Material Random=true
   )]
  particleUpdateScript [
   &ParticleState()
   &ScaleSpriteSize(
    Scale Sprite Size Mode=::NewEnumerator0
    Uniform Scale Factor=&Add_Float(
     A=0.5
     B=&FloatFromCurve(FloatCurve=DataInterface() {propertyValues {Curve {
         defaultValue=0
         keys [
          {arriveTangent=23.9023, interpMode=RCIM_Cubic, leaveTangent=23.9023, tangentMode=RCTM_User, time=0, value=0}
          {arriveTangent=0.9194, interpMode=RCIM_Cubic, leaveTangent=3.5477, tangentMode=RCTM_Break, time=0.2438, value=0.2479}
          {arriveTangent=1.1355, interpMode=RCIM_Cubic, leaveTangent=1.1355, tangentMode=RCTM_User, time=0.4721, value=0.7957}
          {
           arriveTangent=0.7231
           arriveTangentWeight=0.1198
           interpMode=RCIM_Cubic
           leaveTangent=0.7295
           tangentMode=RCTM_Break
           time=0.894
           value=0.9527
          }
          {arriveTangent=-16.5925, arriveTangentWeight=0.0561, interpMode=RCIM_Cubic, leaveTangent=-16.5925, tangentMode=RCTM_User, time=1, value=1}
         ]
        }}})
    )
   ) {module=ScaleSpriteSize002}
   &ScaleColor(
    Scale Alpha=&FloatFromCurve(FloatCurve=DataInterface() {propertyValues {Curve {
        defaultValue=0
        keys [
         {arriveTangent=29.0642, interpMode=RCIM_Cubic, leaveTangent=29.0642, tangentMode=RCTM_User, time=0, value=0}
         {arriveTangent=-0.2441, leaveTangent=-0.2441, time=0.07, value=0.9847}
         {arriveTangent=-1.146, leaveTangent=-1.146, time=1, value=1}
         {arriveTangent=-0.0405, interpMode=RCIM_Cubic, leaveTangent=-0.0405, tangentMode=RCTM_User, time=1, value=0}
        ]
       }}})
    ScaleA=true
    ScaleRGB=true
   )
   &OffsetPosition(Position Offset=$System.BurstPosition)
   &DynamicMaterialParameters(
    Index 0 Param 1=&FloatFromCurve(FloatCurve=DataInterface() {propertyValues {Curve {keys [{arriveTangent=0.6296, interpMode=RCIM_Cubic, leaveTangent=0.6296, tangentMode=RCTM_User, time=0, value=0}, <at 0.75 0.9676>, <at 1 1>]}}})
    Index 0 Param 2=0.2
    Module.Index0Float0Write=true
    Module.Index0Float1Write=true
    Module.Index0Float2Write=true
    Module.Index0Float3Write=true
   )
   /SetVariables(
    Particles.Glow Color=&MakeLinearColorFromVectorAndFloat(
     "Vector (RGB)"=&MakeVectorFromLinearColorRGB(Color=<rgba 1 0.4453 0 1>)
     "Float (Alpha)"=&FloatFromCurve(
      FloatCurve=DataInterface() {propertyValues {Curve {
         defaultValue=1
         keys [
          <at 0.317 0 RCIM_Cubic>
          {arriveTangent=2.3189, interpMode=RCIM_Cubic, leaveTangent=2.3189, time=0.6296, value=0.6459}
          <at 0.7482 1 RCIM_Cubic>
         ]
        }}}
      Scale Curve=0.1
     )
    )
    Particles.Glow Size=&Multiply_Vector2D(A=$Particles.SpriteSize, B=<2.5 2.5>)
   ) {isSetParameters=true}
   &SubUVAnimation(
    End Frame Range Override=7
    Output Alternate Renderer Binding=true
    "Sprite Renderer "=DataInterface() {propertyValues {SpriteRenderer=&self:Intro_0.NiagaraSpriteRendererProperties_2}}
    SubUV Animation Mode:ENiagara_SubUVLookupModeV2=::NewEnumerator1
    SubUV Curve=DataInterface() {propertyValues {Curve {
       defaultValue=0
       keys [
        {arriveTangent=3.3353, interpMode=RCIM_Cubic, leaveTangent=3.3353, tangentMode=RCTM_User, time=0.25, value=0}
        {arriveTangent=2.9618, interpMode=RCIM_Cubic, leaveTangent=2.9618, tangentMode=RCTM_User, time=0.6, value=0.9376}
       ]
      }}}
    UseEndFrame=true
    UseStartFrame=false
   )
  ]
  rendererClasses [&NiagaraSpriteRendererProperties]
  renderers [
   {rendererClass=&NiagaraSpriteRendererProperties}
   {rendererClass=&NiagaraSpriteRendererProperties, rendererIndex=1}
  ]
 }
 Pop {
  emitterUpdateScript [
   &EmitterState()
   &SpawnBurst_Instantaneous(
    Spawn Group=1
    Spawn Time=&Subtract_Float(A=$System.IntroLength, B=0)
    Use Spawn Probability=false
    UseLoopCountLimit=false
   ) {module=SpawnBurst_Instantaneous001}
   &SpawnBurst_Instantaneous(
    Spawn Group=2
    Spawn Time=&Subtract_Float(A=$System.IntroLength, B=0)
    Use Spawn Probability=false
    UseLoopCountLimit=false
   ) {module=SpawnBurst_Instantaneous002}
   &SpawnBurst_Instantaneous(
    Spawn Group=3
    Spawn Time=&Subtract_Float(A=$System.IntroLength, B=0)
    Use Spawn Probability=false
    UseLoopCountLimit=false
   ) {module=SpawnBurst_Instantaneous003}
  ]
  particleSpawnScript [
   /SetVariables(Particles.VisibilityTag=&SpawnGroup()) {isSetParameters=true}
   &InitializeParticle(
    Lifetime=&MergeHelper_FloatSwitch(0=0.07, 1=0.2, 2=0.2, 3=0.15, 4=0, Index=&SpawnGroup(), Num=::NewEnumerator3)
    Position Offset=$System.BurstPosition
    Sprite Rotation Angle=&MergeHelper_FloatSwitch(0=0, 1=0, 2=25, 3=0, 4=0, Index=&SpawnGroup(), Num=::NewEnumerator3)
    Sprite Rotation Mode=::NewEnumerator0
    Sprite Size Mode=::NewEnumerator3
    Uniform Sprite Size=&MakeCustomFloatFromBool(
     Bool=&Boolean_OrOperation(Bool1=$User.bIsFrontEnd, Bool2=$User.bIsFrontEndPreview)
     True Float=&Multiply_Float(A=15, B=$System.FE Mult)
     False Float=15
    )
    UsePositionOffset=true
    Write Material Random=true
   )
   &DynamicMaterialParameters(
    Index 0 Param 4=1
    Module.Index0Float0Write=true
    Module.Index0Float1Write=true
    Module.Index0Float2Write=true
    Module.Index0Float3Write=true
   )
  ]
  particleUpdateScript [
   &ParticleState()
   &ScaleSpriteSize(
    Scale Sprite Size Mode=::NewEnumerator0
    Uniform Scale Factor=&FloatFromCurve(FloatCurve=DataInterface() {propertyValues {Curve {keys [
         {arriveTangent=3.3458, interpMode=RCIM_Cubic, leaveTangent=3.3458, tangentMode=RCTM_User, time=0, value=1}
         {arriveTangent=1.251, tangentMode=RCTM_User, time=1, value=3}
        ]}}})
   ) {module=ScaleSpriteSize002}
   &ScaleColor(
    Scale Alpha=&FloatFromCurve(FloatCurve=DataInterface() {propertyValues {Curve {
        defaultValue=0
        keys [
         {arriveTangent=0.3137, interpMode=RCIM_Cubic, leaveTangent=0.3137, tangentMode=RCTM_User, time=0, value=1}
         {arriveTangent=-2.0958, interpMode=RCIM_Cubic, leaveTangent=-2.0958, tangentMode=RCTM_User, time=1, value=0}
        ]
       }}})
    Scale RGB=&VectorFromCurve(VectorCurve=DataInterface() {propertyValues {
       XCurve {keys [{interpMode=RCIM_Cubic, tangentMode=RCTM_User, time=0, value=1}, {arriveTangent=-0.2294, tangentMode=RCTM_User, time=1, value=0}]}
       YCurve {keys [{interpMode=RCIM_Cubic, tangentMode=RCTM_User, time=0, value=1}, {arriveTangent=-0.2294, tangentMode=RCTM_User, time=1, value=0}]}
       ZCurve {keys [{interpMode=RCIM_Cubic, tangentMode=RCTM_User, time=0, value=1}, {arriveTangent=-0.2294, tangentMode=RCTM_User, time=1, value=0}]}
      }})
    ScaleA=true
    ScaleRGB=true
   )
   &SolveForcesAndVelocity(Clamp Velocity=false, Limit Acceleration=false)
   /SetVariables(Particles.Glow Color=&MakeLinearColorFromVectorAndFloat(
     "Vector (RGB)"=&MakeVectorFromLinearColorRGB(Color=<rgba 2 0.5301 0 1>)
     "Float (Alpha)"=&FloatFromCurve(FloatCurve=DataInterface() {propertyValues {Curve {
         defaultValue=1
         keys [
          {interpMode=RCIM_Cubic, tangentMode=RCTM_User, time=0.0551, value=1}
          {arriveTangent=-1.0583, interpMode=RCIM_Cubic, leaveTangent=-1.0583, time=0.2874, value=0.3688}
          {interpMode=RCIM_Cubic, tangentMode=RCTM_User, time=1, value=0}
         ]
        }}})
    )) {isSetParameters=true}
   /SetVariables(Particles.Ring Dynamic Param:Vector4f=&MakeVector4(
     X=&FloatFromCurve(FloatCurve=DataInterface() {propertyValues {Curve {keys [
          {arriveTangent=-7.2529, interpMode=RCIM_Cubic, leaveTangent=-7.2529, tangentMode=RCTM_User, time=-0.0022, value=1.0993}
          {arriveTangent=0.875, interpMode=RCIM_Cubic, leaveTangent=0.875, tangentMode=RCTM_User, time=0.5, value=0}
          {arriveTangent=3.8908, interpMode=RCIM_Cubic, leaveTangent=3.8908, tangentMode=RCTM_User, time=1, value=1}
         ]}}})
     Y=&FloatFromCurve(FloatCurve=DataInterface() {propertyValues {Curve {keys [
          {arriveTangent=5.2096, interpMode=RCIM_Cubic, leaveTangent=5.2096, tangentMode=RCTM_User, time=0, value=0.5}
          {arriveTangent=0.2132, interpMode=RCIM_Cubic, leaveTangent=0.2132, tangentMode=RCTM_User, time=1, value=3}
         ]}}})
     Z=&FloatFromCurve(FloatCurve=DataInterface() {propertyValues {Curve {
         defaultValue=1
         keys [
          {arriveTangent=0.0994, interpMode=RCIM_Cubic, leaveTangent=0.0994, tangentMode=RCTM_User, time=0, value=1}
          {arriveTangent=-1.6508, interpMode=RCIM_Cubic, leaveTangent=-1.6508, tangentMode=RCTM_User, time=1, value=0}
         ]
        }}})
     W=&FloatFromCurve(
      FloatCurve=DataInterface() {propertyValues {Curve {keys [
          {arriveTangent=2.3563, interpMode=RCIM_Cubic, leaveTangent=2.3563, tangentMode=RCTM_User, time=0, value=0}
          {tangentMode=RCTM_User, time=1, value=1}
         ]}}}
      Scale Curve=2
     )
    )) {isSetParameters=true}
   /SetVariables(
    Particles.Spike Ring Color:LinearColor=&MakeLinearColorFromVectorAndFloat(
     "Vector (RGB)"=&MakeVectorFromLinearColorRGB(Color=<rgba 5 2.53 0 1>)
     "Float (Alpha)"=&FloatFromCurve(FloatCurve=DataInterface() {propertyValues {Curve {
         defaultValue=1
         keys [
          {tangentMode=RCTM_User, time=0, value=1}
          {arriveTangent=-0.0405, interpMode=RCIM_Cubic, leaveTangent=-0.0405, tangentMode=RCTM_User, time=0.7009, value=1}
          {arriveTangent=-8.0089, interpMode=RCIM_Cubic, leaveTangent=-8.0089, tangentMode=RCTM_User, time=1, value=0}
         ]
        }}})
    )
    Particles.Spike Ring Size:Vector3f=&Multiply_VectorByFloat(
     Vector=&Multiply_VectorByFloat(
      Vector=<0.37 0.37 0.37>
      Float=&FloatFromCurve(
       FloatCurve=DataInterface() {propertyValues {Curve {
          defaultValue=0
          keys [
           {arriveTangent=18.4197, interpMode=RCIM_Cubic, leaveTangent=18.4197, tangentMode=RCTM_User, time=0, value=0}
           {arriveTangent=1.1179, interpMode=RCIM_Cubic, leaveTangent=1.1179, tangentMode=RCTM_User, time=0.2869, value=0.6072}
           {arriveTangent=0.0224, interpMode=RCIM_Cubic, leaveTangent=0.0224, tangentMode=RCTM_User, time=0.7359, value=0.9376}
           <at 0.9958 1.7804 RCIM_Cubic>
          ]
         }}}
       Scale Curve=0.85
      )
     )
     Float=&MakeCustomFloatFromBool(
      Bool=&Boolean_OrOperation(Bool1=$User.bIsFrontEnd, Bool2=$User.bIsFrontEndPreview)
      True Float=$System.FE Mult
      False Float=1
     )
    )
   ) {isSetParameters=true}
   /SetVariables(Particles.Heart Pop SIze:Vector2f=&Multiply_Vector2D_ByFloat(
     Float=&MakeCustomFloatFromBool(
      Bool=&Boolean_OrOperation(Bool1=$User.bIsFrontEnd, Bool2=$User.bIsFrontEndPreview)
      True Float=$System.FE Mult
      False Float=1
     )
     Vector2D=&Multiply_Vector2D_ByFloat(
      Float=&FloatFromCurve(FloatCurve=DataInterface() {propertyValues {Curve {
          defaultValue=1
          keys [
           {arriveTangent=4.1785, interpMode=RCIM_Cubic, leaveTangent=4.1785, tangentMode=RCTM_User, time=0, value=0}
           {arriveTangent=1.1179, interpMode=RCIM_Cubic, leaveTangent=1.1179, tangentMode=RCTM_User, time=0.2042, value=0.6228}
           {arriveTangent=0.0224, interpMode=RCIM_Cubic, leaveTangent=0.0224, tangentMode=RCTM_User, time=1, value=1}
          ]
         }}})
      Vector2D=&Vector2DFromFloat(Value=50)
     )
    )) {isSetParameters=true}
   /SetVariables(
    Particles.Impact Color:LinearColor=&MakeLinearColorFromVectorAndFloat(
     "Vector (RGB)"=&MakeVectorFromLinearColorRGB(Color=<rgba 3 1.8493 0 1>)
     "Float (Alpha)"=&FloatFromCurve(
      FloatCurve=DataInterface() {propertyValues {Curve {
         defaultValue=1
         keys [{interpMode=RCIM_Cubic, tangentMode=RCTM_User, time=0, value=1}, {interpMode=RCIM_Cubic, tangentMode=RCTM_User, time=1, value=0}]
        }}}
      Scale Curve=0.15
     )
    )
    Particles.Impact Size:Vector2f=&Vector2DFromFloat(Value=&Multiply_Float(
      A=30
      B=&MakeCustomFloatFromBool(
       Bool=&Boolean_OrOperation(Bool1=$User.bIsFrontEnd, Bool2=$User.bIsFrontEndPreview)
       True Float=$System.FE Mult
       False Float=1
      )
     ))
   ) {isSetParameters=true}
   &SubUVAnimation(
    Direct Frame Index:NiagaraFloat=&FloatFromCurve(FloatCurve=DataInterface() {propertyValues {Curve {
        defaultValue=1
        keys [
         <at 0 0 RCIM_Cubic>
         {arriveTangent=5.4794, interpMode=RCIM_Cubic, leaveTangent=5.4794, time=0.3751, value=2.8496}
         {arriveTangent=5.022, interpMode=RCIM_Cubic, leaveTangent=5.022, time=0.8983, value=4.922}
         <at 1.0024 6 RCIM_Cubic>
        ]
       }}})
    Output Alternate Renderer Binding=true
    "Sprite Renderer "=DataInterface() {propertyValues {SpriteRenderer=&self:Pop_0.NiagaraSpriteRendererProperties_1}}
    SubUV Animation Mode:ENiagara_SubUVLookupModeV2=::NewEnumerator5
   ) {module=SubUVAnimation001}
   /SetVariables(
    Particles.Flare Color:LinearColor=&MakeLinearColorFromVectorAndFloat(
     "Vector (RGB)"=&MakeVectorFromLinearColorRGB(Color=<rgba 6 3.3682 0 1>)
     "Float (Alpha)"=&FloatFromCurve(
      FloatCurve=DataInterface() {propertyValues {Curve {
         defaultValue=1
         keys [
          {interpMode=RCIM_Cubic, tangentMode=RCTM_User, time=0.1298, value=1}
          {arriveTangent=-1.4244, interpMode=RCIM_Cubic, leaveTangent=-1.4244, tangentMode=RCTM_User, time=0.2677, value=0.3979}
          {interpMode=RCIM_Cubic, tangentMode=RCTM_User, time=1, value=0}
         ]
        }}}
      Scale Curve=0.6
     )
    )
    Particles.Flare Size:Vector2f=&Multiply_Vector2D_ByFloat(
     Float=&MakeCustomFloatFromBool(
      Bool=&Boolean_OrOperation(Bool1=$User.bIsFrontEnd, Bool2=$User.bIsFrontEndPreview)
      True Float=$System.FE Mult
      False Float=1
     )
     Vector2D=&Multiply_Vector2D_ByFloat(
      Float=&FloatFromCurve(FloatCurve=DataInterface() {propertyValues {Curve {
          defaultValue=1
          keys [
           <at 0 1 RCIM_Cubic>
           {arriveTangent=-0.8138, interpMode=RCIM_Cubic, leaveTangent=-0.8138, tangentMode=RCTM_User, time=0.6462, value=0.7659}
           {arriveTangent=-0.9513, interpMode=RCIM_Cubic, leaveTangent=-0.9513, tangentMode=RCTM_User, time=1, value=0}
          ]
         }}})
      Vector2D=<50 50>
     )
    )
   ) {isSetParameters=true}
   /SetVariables(
    Particles.Dark Color:LinearColor=&MakeLinearColorFromVectorAndFloat(
     "Vector (RGB)"=&MakeVectorFromLinearColorRGB(Color=<rgba 0.0317 0.0007 0.0688 1>)
     "Float (Alpha)"=&FloatFromCurve(
      FloatCurve=DataInterface() {propertyValues {Curve {
         defaultValue=1
         keys [
          {interpMode=RCIM_Cubic, tangentMode=RCTM_User, time=0.1298, value=1}
          {arriveTangent=-1.4244, interpMode=RCIM_Cubic, leaveTangent=-1.4244, tangentMode=RCTM_User, time=0.3335, value=0.3979}
          {interpMode=RCIM_Cubic, tangentMode=RCTM_User, time=1, value=0}
         ]
        }}}
      Scale Curve=1.25
     )
    )
    Particles.Dark Size:Vector2f=&Multiply_Vector2D_ByFloat(
     Float=&MakeCustomFloatFromBool(
      Bool=&Boolean_OrOperation(Bool1=$User.bIsFrontEnd, Bool2=$User.bIsFrontEndPreview)
      True Float=$System.FE Mult
      False Float=1
     )
     Vector2D=&Multiply_Vector2D_ByFloat(
      Float=&FloatFromCurve(FloatCurve=DataInterface() {propertyValues {Curve {
          defaultValue=0
          keys [
           {arriveTangent=4.1785, interpMode=RCIM_Cubic, leaveTangent=4.1785, tangentMode=RCTM_User, time=0, value=0}
           {arriveTangent=1.1179, interpMode=RCIM_Cubic, leaveTangent=1.1179, tangentMode=RCTM_User, time=0.2042, value=0.6228}
           {arriveTangent=0.0224, interpMode=RCIM_Cubic, leaveTangent=0.0224, tangentMode=RCTM_User, time=0.8202, value=1}
           {arriveTangent=-9.2958, interpMode=RCIM_Cubic, leaveTangent=-9.2958, tangentMode=RCTM_User, time=1.0001, value=0}
          ]
         }}})
      Vector2D=<70 70>
     )
    )
   ) {isSetParameters=true}
  ]
  rendererClasses [&NiagaraSpriteRendererProperties, &NiagaraMeshRendererProperties]
  renderers [
   {rendererClass=&NiagaraSpriteRendererProperties}
   {rendererClass=&NiagaraMeshRendererProperties, rendererIndex=1}
   {rendererClass=&NiagaraSpriteRendererProperties, rendererIndex=2}
   {rendererClass=&NiagaraSpriteRendererProperties, rendererIndex=3}
   {rendererClass=&NiagaraSpriteRendererProperties, rendererIndex=4}
  ]
 }
 lilFlash {
  emitterUpdateScript [
   &EmitterState()
   &SpawnBurst_Instantaneous(
    Spawn Count=2
    Spawn Time=&Subtract_Float(A=$System.IntroLength, B=0)
    Use Spawn Probability=false
    UseLoopCountLimit=false
   )
  ]
  particleSpawnScript [
   &PartitionParticles()
   &InitializeParticle(
    Color=<rgba 1 0.8506 0.455 1>
    Lifetime=0.1
    Position Offset=&AddVectorToPosition(
     Position=&MaskVector(Mask Source=1, Mask Parameter=$Particles.PartitionParticles.Partition, Mask=0, Masked Vector=<-13 0 1>, Passthrough Vector=<15 0 0>)
     Vector=$System.BurstPosition
    )
    Sprite Rotation Angle=&MaskFloat(Mask Source=1, Mask Parameter=$Particles.PartitionParticles.Partition, Mask=0, Masked Float=-80, Passthrough Float=100)
    Sprite Rotation Mode=::NewEnumerator0
    Sprite Size Mode=::NewEnumerator3
    Uniform Sprite Size=&Multiply_Float(
     A=&MakeCustomFloatFromBool(
      Bool=&Boolean_OrOperation(Bool1=$User.bIsFrontEnd, Bool2=$User.bIsFrontEndPreview)
      True Float=$System.FE Mult
      False Float=1
     )
     B=33
    )
    UsePositionOffset=true
    Write Material Random=true
   )
   &SubUVAnimation(
    Direct Frame Index=8
    "Sprite Renderer "=DataInterface() {propertyValues {SpriteRenderer=&self:lilFlash_0.NiagaraSpriteRendererProperties_2}}
    SubUV Animation Mode:ENiagara_SubUVLookupModeV2=::NewEnumerator5
   )
   &AddVelocity(Velocity=&MaskVector(Mask Source=1, Mask Parameter=$Particles.PartitionParticles.Partition, Mask=0, Masked Vector=<-50 0 0>, Passthrough Vector=<50 0 0>)) {enabled=false}
   &ApplyInitialForces()
  ]
  particleUpdateScript [
   &ParticleState()
   &ScaleSpriteSize(Uniform Curve Sprite Scale=DataInterface() {propertyValues {Curve {
       defaultValue=0
       keys [
        {arriveTangent=4.1785, interpMode=RCIM_Cubic, leaveTangent=4.1785, tangentMode=RCTM_User, time=0.0083, value=0.5463}
        {arriveTangent=1.1179, interpMode=RCIM_Cubic, leaveTangent=1.1179, tangentMode=RCTM_User, time=0.2475, value=1.1535}
        {arriveTangent=0.0224, interpMode=RCIM_Cubic, leaveTangent=0.0224, tangentMode=RCTM_User, time=1, value=1.6712}
       ]
      }}}) {enabled=false}
   &Drag(Drag=10, Use Linear Drag=true, Use Rotational Drag=true)
   &SolveForcesAndVelocity(Clamp Velocity=false, Limit Acceleration=false)
   &ScaleColor(
    Scale Alpha=&FloatFromCurve(FloatCurve=DataInterface() {propertyValues {Curve {
        defaultValue=0
        keys [
         {interpMode=RCIM_Cubic, tangentMode=RCTM_User, time=0.0701, value=1}
         {arriveTangent=-1.0754, interpMode=RCIM_Cubic, leaveTangent=-1.0754, time=0.1706, value=0.3286}
         {interpMode=RCIM_Cubic, tangentMode=RCTM_User, time=1, value=0}
        ]
       }}})
    ScaleA=true
    ScaleRGB=true
   )
  ]
  rendererClasses [&NiagaraSpriteRendererProperties]
  renderers [{rendererClass=&NiagaraSpriteRendererProperties}]
 }
}
summary {
 systemName=NS_ReactFX_BurgleBoss
 emitters [
  {emitterName=Intro, rendererClasses [&NiagaraSpriteRendererProperties]}
  {emitterName=Pop, rendererClasses [&NiagaraSpriteRendererProperties, &NiagaraMeshRendererProperties]}
  {emitterName=Gold_Pieces, rendererClasses [&NiagaraSpriteRendererProperties]}
  {emitterName=GoldDust, rendererClasses [&NiagaraSpriteRendererProperties]}
  {emitterName=GoldDust001, rendererClasses [&NiagaraSpriteRendererProperties]}
  {emitterName=lilFlash, rendererClasses [&NiagaraSpriteRendererProperties]}
 ]
 userVariables [
  User.bIsFrontEnd {type=NiagaraBool}
  User.bIsFrontEndPreview {type=NiagaraBool}
 ]
}
