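# Niagara reaction effect for the BurrBaby cosmetic companion, pairing a pulsing heart with confetti and a petting reaction sound.
# NE_Heart plays the audio and fades two differently sized sprite groups through scale, glow and material curves, with time-of-day glow scaling.
# Confetti swirls inward via attraction and vortex forces, while a second burst fires outward near IntroLength's end.
# Shared system variables drive timing and positions; lightweight AirGlow, Rays and Corona emitters list no details.
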NiagaraSystem /CosmeticCompanions/FX/React/NS_ReactFX_BurrBaby
dependencies {
 usedDataInterfaces [&NiagaraDataInterfaceAudioPlayer, &NiagaraDataInterfaceCurve]
 usedDynamicInputs [
  &RandomRangeLinearColor
  &SimulationPosition
  /Niagara/DynamicInputs/TypeConversions/MakeCustomPositionFromBool
 ]
 usedModules [
  &self:SystemSpawnScript.SystemScriptSource.SystemScriptGraph.NiagaraNodeAssignment_0.SetVariables_#1
  &self:SystemSpawnScript.SystemScriptSource.SystemScriptGraph.NiagaraNodeAssignment_4.SetVariables_#2
  &self:SystemSpawnScript.SystemScriptSource.SystemScriptGraph.NiagaraNodeAssignment_1.SetVariables_#3
  &self:SystemSpawnScript.SystemScriptSource.SystemScriptGraph.NiagaraNodeAssignment_6.SetVariables_#4
  &self:SystemSpawnScript.SystemScriptSource.SystemScriptGraph.NiagaraNodeAssignment_7.SetVariables_#5
  &self:SystemSpawnScript.SystemScriptSource.SystemScriptGraph.NiagaraNodeAssignment_2.SetVariables_#6
  &self:SystemSpawnScript.SystemScriptSource.SystemScriptGraph.NiagaraNodeAssignment_3.SetVariables_#7
  &self:SystemSpawnScript.SystemScriptSource.SystemScriptGraph.NiagaraNodeAssignment_5.SetVariables_#8
  &SystemState
 ]
}
emitters {
 LWE_AirGlow {rendererClasses [], renderers []}
 LWE_Corona {rendererClasses [], renderers []}
 LWE_Rays {rendererClasses [], renderers []}
 NE_Confetti {
  emitterUpdateScript [
   &EmitterState()
   &SpawnBurst_Instantaneous(Spawn Count=35, Spawn Time=&Subtract_Float(A=$System.IntroLength, B=0.05), Use Spawn Probability=false)
  ]
  particleSpawnScript [
   &InitializeParticle(
    Color Maximum=<rgba 0.75 0.305 0 1>
    Color Minimum=<rgba 0.5 0.305 0 1>
    Color Mode=::NewEnumerator2
    Lifetime Max=2.5
    Lifetime Min=0.2
    Lifetime Mode=::NewEnumerator1
    Position Offset:NiagaraPosition=$System.BurstPosition
    Position Offset Coordinate Space=::Local
    Sprite Rotation Mode=::NewEnumerator2
    Sprite Size Mode=::NewEnumerator1
    Uniform Sprite Size Max=6
    Uniform Sprite Size Min=3
    UsePositionOffset=true
    Write Material Random=true
   ) {module=InitializeParticle001}
   &AddVelocity(
    Module.Use Falloff Distance=false
    Origin Offset:NiagaraPosition=$System.BurstPosition
    Override Seed=false
    Velocity Mode=::NewEnumerator1
    Velocity Speed=&RandomRangeFloat(Minimum=150, Maximum=300)
   ) {module=AddVelocity001}
  ]
  particleUpdateScript [
   &ParticleState()
   &Drag(Drag=&RandomRangeFloat(Minimum=5, Maximum=10), Use Linear Drag=true, Use Rotational Drag=false)
   &GravityForce(Gravity=&RandomRangeVector(Minimum=<0 0 -360>, Maximum=<1 1 -200>))
   &SolveForcesAndVelocity(Clamp Velocity=false, Limit Acceleration=false)
   &ScaleSpriteSize(Scale Factor=&Vector2DFromFloat(Value=&FloatFromCurve(FloatCurve=DataInterface() {propertyValues {Curve {keys [
          {arriveTangent=-0.4596, interpMode=RCIM_Cubic, leaveTangent=-0.4596, tangentMode=RCTM_User, time=0.55, value=1}
          {arriveTangent=-4.9031, interpMode=RCIM_Cubic, leaveTangent=-4.9031, tangentMode=RCTM_User, time=1, value=0}
         ]}}})))
   &ScaleColor(
    Scale RGB=&MakeVector(
     X=1
     Y=&Add_Float(
      A=&RandomRangeFloat(Minimum=0, Maximum=1000)
      B=&Lerp_Float(A=0, B=&RandomRangeFloat(Minimum=25, Maximum=150), Alpha=$Particles.NormalizedAge)
     )
     Z=1
    )
    ScaleA=false
    ScaleRGB=true
   )
  ]
  rendererClasses [&NiagaraSpriteRendererProperties]
  renderers [{rendererClass=&NiagaraSpriteRendererProperties}]
 }
 NE_Confetti_In {
  emitterUpdateScript [
   &EmitterState(Life Cycle Mode=::NewEnumerator1, Loop Behavior=::NewEnumerator1, Loop Duration=0.5, UseLoopDelay=false)
   &SpawnRate(SpawnRate=100, Use Spawn Probability=false)
  ]
  particleSpawnScript [
   &InitializeParticle(
    Color Maximum=<rgba 0.75 0.305 0 1>
    Color Minimum=<rgba 0.5 0.305 0 1>
    Color Mode=::NewEnumerator2
    Lifetime Max=0.4
    Lifetime Min=0.2
    Lifetime Mode=::NewEnumerator1
    Position Offset=&Subtract_Position(A=$System.BurstPosition, B=&FN_FrontEndVector3Modifier(Gameplay=<0 0 0>, Front End=<0 0 0>))
    Position Offset Coordinate Space=::Local
    Sprite Rotation Mode=::NewEnumerator2
    Sprite Size Mode=::NewEnumerator1
    Uniform Sprite Size Max=15
    UsePositionOffset=true
    Write Material Random=true
   ) {module=InitializeParticle001}
   &ShapeLocation(MaskSpawnGroup=false, Override Seed=false, Sphere Radius=80, UseOwnerScale=true)
  ]
  particleUpdateScript [
   &ParticleState()
   &PointAttractionForce(
    Attraction Radius=500
    Attraction Strength=500
    Attractor Position Offset:Vector3f=&Subtract_Position(A=$System.BurstPosition, B=&FN_FrontEndVector3Modifier(Gameplay=<0 0 0>, Front End=<0 0 0>))
    Attractor Position Offset Coordinate Space:ENiagaraCoordinateSpace=::Local
    Kill Within Radius=false
    Module.Use Falloff=true
    UseAttractorOffset=true
   )
   &Drag(Drag=&RandomRangeFloat(Minimum=5, Maximum=20), Use Linear Drag=true, Use Rotational Drag=false)
   &VortexForce(
    Origin Pull Amount=1500
    UseOriginOffset=true
    Vortex Axis=<0 1 0>
    Vortex Axis Coordinate Space:ENiagaraCoordinateSpace=::Local
    Vortex Force Amount=3000
    Vortex Origin Offset:Vector3f=&Subtract_Position(A=$System.BurstPosition, B=&FN_FrontEndVector3Modifier(Gameplay=<0 0 0>, Front End=<0 0 0>))
    Vortex Origin Offset Coordinate Space:ENiagaraCoordinateSpace=::Local
   )
   &ScaleSpriteSize(Uniform Curve Sprite Scale=DataInterface() {propertyValues {Curve {
       defaultValue=0
       keys [
        {interpMode=RCIM_Cubic, tangentMode=RCTM_User, time=0, value=1}
        {arriveTangent=-2.3313, interpMode=RCIM_Cubic, leaveTangent=-2.3313, tangentMode=RCTM_User, time=1, value=0}
       ]
      }}}) {module=ScaleSpriteSize001}
   &SolveForcesAndVelocity(Clamp Velocity=false, Limit Acceleration=false)
   &ScaleSpriteSize(Scale Factor=&Vector2DFromFloat(Value=&FloatFromCurve(FloatCurve=DataInterface() {propertyValues {Curve {keys [
          {arriveTangent=-0.4596, interpMode=RCIM_Cubic, leaveTangent=-0.4596, tangentMode=RCTM_User, time=0.55, value=1}
          {arriveTangent=-4.9031, interpMode=RCIM_Cubic, leaveTangent=-4.9031, tangentMode=RCTM_User, time=1, value=0}
         ]}}})))
   &ScaleColor(
    Scale RGB=&MakeVector(
     X=1
     Y=&Add_Float(
      A=&RandomRangeFloat(Minimum=0, Maximum=1000)
      B=&Lerp_Float(A=0, B=&RandomRangeFloat(Minimum=25, Maximum=150), Alpha=$Particles.NormalizedAge)
     )
     Z=1
    )
    ScaleA=false
    ScaleRGB=true
   )
  ]
  rendererClasses [&NiagaraSpriteRendererProperties]
  renderers [{rendererClass=&NiagaraSpriteRendererProperties}]
 }
 NE_Heart {
  emitterSpawnScript [
   &InitializeAudioPlayer(Audio Player=DataInterface() {propertyValues {
      Attenuation=null
      Concurrency=null
      MaxPlaysPerTick=10
      SoundToPlay=/CosmeticCompanions/Assets/Audio_Shared/Interaction/Petting/SC_Mimosa_Interaction_React_Default
      bAllowLoopingOneShotSounds=false
      bLimitPlaysPerTick=true
      bOnlyActiveDuringGameplay=false
      bStopWhenComponentIsDestroyed=true
      ConfigurationUserParameter {parameter {name=None, typeDefHandle {registeredTypeIndex=1}, varData []}}
      ParameterNames []
     }})
   /Niagara/Modules/Audio/PlayAudio() {module=PlayAudio001}
  ]
  emitterUpdateScript [
   &EmitterState()
   &SpawnBurst_Instantaneous(Use Spawn Probability=false, UseLoopCountLimit=false)
   &SpawnBurst_Instantaneous(Spawn Group=1, Use Spawn Probability=false, UseLoopCountLimit=false) {module=SpawnBurst_Instantaneous001}
  ]
  particleSpawnScript [
   /SetVariables(Particles.VisibilityTag=&SpawnGroup()) {isSetParameters=true}
   &InitializeParticle(
    Lifetime=$System.IntroLength
    Sprite Rotation Mode=::NewEnumerator0
    Sprite Size=&MergeHelper_Vector2Switch(0=<11 10>, 1=<18 18>, 2=?, 3=?, 4=?, Index=$Particles.VisibilityTag, Num=::NewEnumerator0)
    Sprite Size Mode=::NewEnumerator4
    UsePositionOffset=false
    Write Material Random=true
   )
   /SetVariables(Particles.Glow Color=<rgba 0.7314 0 1 0.35>) {isSetParameters=true}
  ]
  particleUpdateScript [
   &ParticleState()
   &ScaleSpriteSize(
    Scale Sprite Size Mode=::NewEnumerator0
    Uniform Scale Factor=&Add_Float(
     A=0.5
     B=&FloatFromCurve(FloatCurve=DataInterface() {propertyValues {Curve {
         defaultValue=0
         keys [
          {arriveTangent=0.8193, time=0, value=0}
          <at 0.3 2>
          <at 0.5 1.5>
          <at 0.7 2>
          {arriveTangent=-0.9053, interpMode=RCIM_Cubic, leaveTangent=-0.9053, tangentMode=RCTM_User, time=0.8783, value=3}
          {arriveTangent=-0.1202, interpMode=RCIM_Cubic, leaveTangent=-0.1202, tangentMode=RCTM_User, time=1, value=1}
         ]
        }}})
    )
   ) {module=ScaleSpriteSize002}
   /SetVariables(Particles.UpdateGlowColor:LinearColor=&MakeLinearColorFromVectorAndFloat(
     "Vector (RGB)"=&MakeVectorFromLinearColorRGB(Color=$Particles.Initial.Glow Color)
     "Float (Alpha)"=&Multiply_Float(
      A=&FloatFromCurve(
       FloatCurve=DataInterface() {propertyValues {Curve {keys [
           {arriveTangent=11.9433, time=0, value=0}
           {arriveTangent=2, interpMode=RCIM_Cubic, leaveTangent=2, tangentMode=RCTM_User, time=0.5, value=0}
           {arriveTangent=5, interpMode=RCIM_Cubic, leaveTangent=5, time=0.6, value=3}
           <at 0.7 1 RCIM_Cubic>
          ]}}}
       Scale Curve=0.5
      )
      B=&Multiply_Float(A=1, B=&FNTimeOfDayScalar(Night Scale=0.5, Morning Scale=0.25, Evening Scale=0.25))
     )
    )) {isSetParameters=true}
   &ScaleColor(
    Scale Alpha=&FloatFromCurve(FloatCurve=DataInterface() {propertyValues {Curve {
        defaultValue=0
        keys [
         {arriveTangent=9.0725, interpMode=RCIM_Cubic, leaveTangent=9.0725, tangentMode=RCTM_User, time=0, value=0}
         {arriveTangent=-0.2441, leaveTangent=-0.2441, time=0.1408, value=0.9847}
         {arriveTangent=-1.146, leaveTangent=-1.146, time=0.95, value=1}
         {arriveTangent=-0.0405, interpMode=RCIM_Cubic, leaveTangent=-0.0405, tangentMode=RCTM_User, time=1, value=0}
        ]
       }}})
    ScaleA=true
    ScaleRGB=true
   )
   &OffsetPosition(Position Offset=&Lerp_Vector(
     A=$System.HeartIntroSpawnPosition
     B=Linked() {linkedVariable {name=System.BurstPosition, type=NiagaraPosition}}
     Alpha=&FloatFromCurve(FloatCurve=DataInterface() {propertyValues {Curve {keys [{arriveTangent=0.7913, time=0, value=1}, <at 0.3 1>, {arriveTangent=-0.1074, tangentMode=RCTM_User, time=1, value=1}]}}})
    ))
   &DynamicMaterialParameters(
    Index 0 Param 1=&FloatFromCurve(FloatCurve=DataInterface() {propertyValues {Curve {keys [
         <at 0 -1>
         {interpMode=RCIM_Cubic, tangentMode=RCTM_User, time=0.4, value=-1}
         {arriveTangent=19.8658, interpMode=RCIM_Cubic, leaveTangent=19.8658, tangentMode=RCTM_User, time=0.6, value=1}
        ]}}})
    Index 0 Param 2=1
    Index 0 Param 3=&FloatFromCurve(FloatCurve=DataInterface() {propertyValues {Curve {
        defaultValue=1
        keys [
         {tangentMode=RCTM_User, time=0, value=-0.5}
         {interpMode=RCIM_Cubic, tangentMode=RCTM_User, time=0.2, value=-0.5}
         {arriveTangent=15.0323, interpMode=RCIM_Cubic, leaveTangent=15.0323, tangentMode=RCTM_User, time=0.4, value=1}
        ]
       }}})
    Module.Index0Float0Write=true
    Module.Index0Float1Write=true
    Module.Index0Float2Write=true
    Module.Index0Float3Write=true
   )
   &SpriteRotationRate(Rotation Rate=&MergeHelper_FloatSwitch(0=0, 1=90, 2=?, 3=?, Index=$Particles.VisibilityTag, Num=::NewEnumerator0))
  ]
  rendererClasses [&NiagaraSpriteRendererProperties]
  renderers [
   {rendererClass=&NiagaraSpriteRendererProperties}
   {rendererClass=&NiagaraSpriteRendererProperties, rendererIndex=1}
   {rendererClass=&NiagaraSpriteRendererProperties, rendererIndex=2}
  ]
 }
}
summary {
 systemName=NS_ReactFX_BurrBaby
 emitters [
  {emitterName=NE_Heart, rendererClasses [&NiagaraSpriteRendererProperties]}
  {emitterName=NE_Confetti, rendererClasses [&NiagaraSpriteRendererProperties]}
  {emitterName=LWE_AirGlow, rendererClasses []}
  {emitterName=LWE_Rays, rendererClasses []}
  {emitterName=LWE_Corona, rendererClasses []}
  {emitterName=NE_Confetti_In, rendererClasses [&NiagaraSpriteRendererProperties]}
 ]
 userVariables [
  User.bIsFrontEnd {type=NiagaraBool}
  User.bIsFrontEndPreview {type=NiagaraBool}
 ]
}
