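# Niagara reaction effect for the CacaoLynx cosmetic companion, pairing a glowing heart with the shared petting reaction sound.
# A projectile sprite travels from portal to burst position amid falling streaks, while sparks surround the growing heart.
# A delayed finale fires a mesh shockwave with time-of-day-scaled alpha and staggered flipbook flames.
# Front-end modifiers adjust placement and scale, local GimbalLock modules use camera vectors, and the ribbon Laser emitter is disabled.
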NiagaraSystem /CosmeticCompanions/FX/React/NS_ReactFX_CacaoLynx
dependencies {
 usedDataInterfaces [
  &NiagaraDataInterfaceAudioPlayer
  &NiagaraDataInterfaceCurve
  &NiagaraDataInterfaceCamera
  &NiagaraDataInterfaceColorCurve
  &NiagaraDataInterfaceVector2DCurve
  &NiagaraDataInterfaceSpriteRendererInfo
 ]
 usedDynamicInputs [&MakeVector4_FromLinearColor, &FNTimeOfDayScalar]
 usedModules [
  &self:SystemSpawnScript.SystemScriptSource.SystemScriptGraph.NiagaraNodeAssignment_7.SetVariables_#1
  &self:SystemSpawnScript.SystemScriptSource.SystemScriptGraph.NiagaraNodeAssignment_8.SetVariables_#2
  &self:SystemSpawnScript.SystemScriptSource.SystemScriptGraph.NiagaraNodeAssignment_12.SetVariables_#3
  &self:SystemSpawnScript.SystemScriptSource.SystemScriptGraph.NiagaraNodeAssignment_9.SetVariables_#4
  &self:SystemSpawnScript.SystemScriptSource.SystemScriptGraph.NiagaraNodeAssignment_10.SetVariables_#5
  &self:SystemSpawnScript.SystemScriptSource.SystemScriptGraph.NiagaraNodeAssignment_2.SetVariables_#6
  &self:SystemSpawnScript.SystemScriptSource.SystemScriptGraph.NiagaraNodeAssignment_3.SetVariables_#7
  &self:SystemSpawnScript.SystemScriptSource.SystemScriptGraph.NiagaraNodeAssignment_6.SetVariables_#8
  &self:SystemSpawnScript.SystemScriptSource.SystemScriptGraph.NiagaraNodeAssignment_0.SetVariables_#9
  &SystemState
  &self:SystemSpawnScript.SystemScriptSource.SystemScriptGraph.NiagaraNodeAssignment_5.SetVariables_#10
  &self:SystemSpawnScript.SystemScriptSource.SystemScriptGraph.NiagaraNodeAssignment_11.SetVariables_#11
  &self:SystemSpawnScript.SystemScriptSource.SystemScriptGraph.NiagaraNodeAssignment_1.SetVariables_#12
 ]
}
emitters {
 Empty {rendererClasses [], renderers []}
 Empty001 {rendererClasses [], renderers []}
 Fire002 {
  emitterUpdateScript [
   &EmitterState(Life Cycle Mode=::NewEnumerator1, Loop Behavior=::NewEnumerator1, Loop Delay=0.72, Loop Duration=1, UseLoopDelay=true)
   &SpawnBurst_Instantaneous(Use Spawn Probability=false, UseLoopCountLimit=false)
   &SpawnBurst_Instantaneous(Spawn Group=1, Spawn Time=0.04, Use Spawn Probability=false, UseLoopCountLimit=false) {module=SpawnBurst_Instantaneous001}
   &SpawnBurst_Instantaneous(Spawn Group=2, Spawn Time=0.08, Use Spawn Probability=false, UseLoopCountLimit=false) {module=SpawnBurst_Instantaneous002}
   &SpawnBurst_Instantaneous(Spawn Group=3, Spawn Time=0.12, Use Spawn Probability=false, UseLoopCountLimit=false) {module=SpawnBurst_Instantaneous003}
   &SpawnBurst_Instantaneous(Spawn Group=4, Spawn Time=0.16, Use Spawn Probability=false, UseLoopCountLimit=false) {module=SpawnBurst_Instantaneous004}
  ]
  particleSpawnScript [
   /SetVariables(Particles.VisibilityTag=&SpawnGroup()) {isSetParameters=true}
   &InitializeParticle(
    Color=<rgba 1 0.8774 0.849 1>
    Lifetime=&MergeHelper_FloatSwitch(0=0.2, 1=0.28, 2=0.28, 3=0.28, 4=0.28, Index=&SpawnGroup(), Num=::NewEnumerator3)
    Sprite Size=&MergeHelper_Vector2Switch(
     0=<35 50>
     1=&Vector2DFromFloat(Value=10)
     2=&Vector2DFromFloat(Value=15)
     3=&Vector2DFromFloat(Value=22.5)
     4=&Vector2DFromFloat(Value=15)
     Index=&SpawnGroup()
     Num=::NewEnumerator3
    )
    Sprite Size Mode=::NewEnumerator4
    UsePositionOffset=false
    Write Material Random=true
   )
  ]
  particleUpdateScript [
   &ParticleState()
   &ScaleSpriteSize(
    Scale Factor=&MergeHelper_Vector2Switch(
     0=&Vector2DFromCurve(
      Vector2Curve=DataInterface() {propertyValues {
        XCurve {keys [
          <at 0 0 RCIM_Cubic>
          {arriveTangent=0.0112, interpMode=RCIM_Cubic, leaveTangent=0.0112, tangentMode=RCTM_User, time=0.2, value=1}
          {arriveTangent=-0.1528, interpMode=RCIM_Cubic, leaveTangent=-0.1528, tangentMode=RCTM_User, time=0.35, value=1}
          <at 1 1 RCIM_Cubic>
         ]}
        YCurve {keys [<at 0 0 RCIM_Cubic>, {arriveTangent=0.1288, interpMode=RCIM_Cubic, leaveTangent=0.1288, tangentMode=RCTM_User, time=0.5, value=1}]}
       }}
      ScaleCurve=<1 1>
     )
     1=<1 1>
     2=<1 1>
     3=<1 1>
     4=<1 1>
     Index=&SpawnGroup()
     Num=::NewEnumerator3
    )
    Scale Sprite Size Mode=::NewEnumerator2
   )
   &SubUVAnimation(
    Play Rate=4
    "Sprite Renderer "=DataInterface() {propertyValues {SpriteRenderer=&self:Fire002_0.NiagaraSpriteRendererProperties_0}}
    SubUV Animation Mode:ENiagara_SubUVLookupModeV2=::NewEnumerator3
    UseEndFrame=false
    UsePlayRateScale=false
    UseStartFrame=false
   )
   &ScaleColor(
    Scale Alpha=&FloatFromCurve(
     FloatCurve=DataInterface() {propertyValues {Curve {
        defaultValue=1
        keys [
         {arriveTangent=9.0725, interpMode=RCIM_Cubic, leaveTangent=9.0725, tangentMode=RCTM_User, time=0, value=0}
         {arriveTangent=-0.0958, interpMode=RCIM_Cubic, leaveTangent=-0.0958, tangentMode=RCTM_User, time=0.1408, value=0.9847}
         {arriveTangent=-1.9837, interpMode=RCIM_Cubic, leaveTangent=-1.9837, tangentMode=RCTM_User, time=1, value=0}
        ]
       }}}
     Scale Curve=&MaskFloat(Mask Source=0, Mask Parameter=?, Mask=0, Masked Float=1, Passthrough Float=0.7)
    )
    Scale RGB=&MaskVector(
     Mask Source=0
     Mask Parameter=?
     Mask=0
     Masked Vector=<1 1 1>
     Passthrough Vector=&MakeVectorFromLinearColorRGB(Color=&ColorFromCurve(ColorCurve=DataInterface() {propertyValues {
         AlphaCurve {keys [<at 0 1 RCIM_Cubic>,<at 1 0 RCIM_Cubic>]}
         BlueCurve {keys [<at 0 0 RCIM_Cubic>,<at 0.25 0.0267>,<at 0.5 0>,<at 1 0>]}
         GreenCurve {keys [<at 0 0.0441 RCIM_Cubic>,<at 0.25 0.6473>,<at 0.5 0.0697>,<at 1 0.0126>]}
         RedCurve {keys [<at 0 1 RCIM_Cubic>,<at 0.25 0.9056>,<at 0.5 0.6784>,<at 1 0.2865>]}
        }}))
    )
    ScaleA=true
    ScaleRGB=true
   )
   /SetVariables(Particles.Position=&MergeHelper_VectorSwitch(
     0=&Lerp_Vector(
      A=$System.BurstPosition
      B=&FN_FrontEndVector3Modifier(Gameplay=<0 0 25>, Front End=<0 0 -30>)
      Alpha=&FloatFromCurve(FloatCurve=DataInterface() {propertyValues {Curve {keys [<at 0 0>,<at 1 1>]}}})
     )
     1=&FN_FrontEndVector3Modifier(
      Gameplay=<0 0 15>
      Front End=&AddVectorToPosition(Position=Linked() {linkedVariable {name=System.BurstPosition, type=Vector3f}}, Vector=<0 0 5>)
     )
     2=&FN_FrontEndVector3Modifier(
      Gameplay=<0 0 20>
      Front End=&AddVectorToPosition(Position=Linked() {linkedVariable {name=System.BurstPosition, type=Vector3f}}, Vector=<0 0 10>)
     )
     3=&FN_FrontEndVector3Modifier(
      Gameplay=<0 0 30>
      Front End=&AddVectorToPosition(Position=Linked() {linkedVariable {name=System.BurstPosition, type=Vector3f}}, Vector=<0 0 20>)
     )
     4=&FN_FrontEndVector3Modifier(
      Gameplay=<0 0 45>
      Front End=&AddVectorToPosition(Position=Linked() {linkedVariable {name=System.BurstPosition, type=Vector3f}}, Vector=<0 0 35>)
     )
     Index=&SpawnGroup()
     Num=::NewEnumerator3
    )) {isSetParameters=true}
   &self:GimbalLock_01(
    Gimbal Lock Direction:NiagaraPosition=&ReturnCameraVector(Camera Vectors=::NewEnumerator4, Coordinate Space=::Local, Origin Position=$Particles.Position)
    Position:NiagaraPosition=$Particles.Position
   ) {module=GimbalLock_01}
   &SubUVAnimation(
    "Sprite Renderer "=DataInterface() {propertyValues {SpriteRenderer=&self:Fire002_0.NiagaraSpriteRendererProperties_1}}
    UseEndFrame=false
    UseStartFrame=false
   ) {module=SubUVAnimation001}
   &SubUVAnimation(
    "Sprite Renderer "=DataInterface() {propertyValues {SpriteRenderer=&self:Fire002_0.NiagaraSpriteRendererProperties_2}}
    UseEndFrame=false
    UseStartFrame=false
   ) {module=SubUVAnimation002}
   &SubUVAnimation(
    "Sprite Renderer "=DataInterface() {propertyValues {SpriteRenderer=&self:Fire002_0.NiagaraSpriteRendererProperties_3}}
    UseEndFrame=false
    UseStartFrame=false
   ) {module=SubUVAnimation003}
   &SubUVAnimation(
    "Sprite Renderer "=DataInterface() {propertyValues {SpriteRenderer=&self:Fire002_0.NiagaraSpriteRendererProperties_4}}
    UseEndFrame=false
    UseStartFrame=false
   ) {module=SubUVAnimation004}
  ]
  rendererClasses [&NiagaraSpriteRendererProperties]
  renderers [
   {rendererClass=&NiagaraSpriteRendererProperties}
   {rendererClass=&NiagaraSpriteRendererProperties, rendererIndex=1}
   {rendererClass=&NiagaraSpriteRendererProperties, rendererIndex=2}
   {rendererClass=&NiagaraSpriteRendererProperties, rendererIndex=3}
   {rendererClass=&NiagaraSpriteRendererProperties, rendererIndex=4}
  ]
 }
 Heart {
  emitterSpawnScript [
   &InitializeAudioPlayer(Audio Player=DataInterface() {propertyValues {
      Attenuation=null
      Concurrency=null
      MaxPlaysPerTick=10
      SoundToPlay=/CosmeticCompanions/Assets/Audio_Shared/Interaction/Petting/SC_Mimosa_Interaction_React_Default
      bAllowLoopingOneShotSounds=false
      bLimitPlaysPerTick=true
      bOnlyActiveDuringGameplay=false
      bStopWhenComponentIsDestroyed=true
      ConfigurationUserParameter {parameter {name=None, typeDefHandle {registeredTypeIndex=1}, varData []}}
      ParameterNames []
     }})
   /Niagara/Modules/Audio/PlayAudio() {module=PlayAudio001}
  ]
  emitterUpdateScript [
   &EmitterState(Life Cycle Mode=::NewEnumerator1, Loop Behavior=::NewEnumerator1, Loop Delay=0.1, Loop Duration=1, UseLoopDelay=true)
   &SpawnBurst_Instantaneous(Use Spawn Probability=false, UseLoopCountLimit=false)
  ]
  particleSpawnScript [
   &InitializeParticle(
    Lifetime=0.68
    Position Offset=&AddVectorToPosition(
     Position=Linked() {linkedVariable {name=System.BurstPosition, type=Vector3f}}
     Vector=&FN_FrontEndVector3Modifier(Gameplay=<0 0 12>, Front End=<0 0 -6>)
    )
    Sprite Rotation Mode=::NewEnumerator0
    Sprite Size Mode=::NewEnumerator3
    Uniform Sprite Size=&Multiply_Float(A=64, B=&FN_FrontEndFloatModifier(Gameplay=1, Front End=0.8))
    UsePositionOffset=true
    Write Material Random=true
   )
   &CameraOffset()
  ]
  particleUpdateScript [
   &ParticleState()
   &ScaleSpriteSize(
    Scale Sprite Size Mode=::NewEnumerator0
    Uniform Scale Factor=&Add_Float(
     A=0
     B=&FloatFromCurve(FloatCurve=DataInterface() {propertyValues {Curve {
         defaultValue=0
         keys [
          {arriveTangent=0.7072, interpMode=RCIM_Cubic, leaveTangent=0.7072, tangentMode=RCTM_User, time=0, value=0.85}
          {arriveTangent=0.1607, interpMode=RCIM_Cubic, leaveTangent=0.1607, tangentMode=RCTM_User, time=0.8159, value=1.0029}
          {arriveTangent=-0.1285, interpMode=RCIM_Cubic, leaveTangent=-0.1285, tangentMode=RCTM_User, time=0.9003, value=1}
          <at 1.0003 0 RCIM_Cubic>
         ]
        }}})
    )
   ) {module=ScaleSpriteSize002}
   &self:GimbalLock_02(
    Gimbal Lock Direction:NiagaraPosition=&ReturnCameraVector(Camera Vectors=::NewEnumerator4, Coordinate Space=::Local, Origin Position=$Particles.Position)
    Position:NiagaraPosition=$Particles.Position
   ) {module=GimbalLock_02}
   /SetVariables(
    Particles.Glow Color=&Multiply_LinearColorByFloat(
     Linear Color=<rgba 10.3043 4.1944 1.0847 0.25>
     Float=&FloatFromCurve(FloatCurve=DataInterface() {propertyValues {Curve {
         defaultValue=0
         keys [
          {arriveTangent=0.2982, interpMode=RCIM_Cubic, leaveTangent=0.2982, tangentMode=RCTM_User, time=0.12, value=0}
          {interpMode=RCIM_Cubic, tangentMode=RCTM_User, time=0.95, value=1}
          <at 1 0 RCIM_Cubic>
         ]
        }}})
    )
    Particles.Glow Pos:NiagaraPosition=&AddVectorToPosition(
     Position=$Particles.Position
     Vector=&FN_FrontEndVector3Modifier(Gameplay=<0 0 -2>, Front End=<0 0 -1.4>)
    )
    Particles.Glow Size=&Multiply_Vector2D_ByFloat(Float=5, Vector2D=<1 1>)
   ) {isSetParameters=true}
  ]
  rendererClasses [&NiagaraSpriteRendererProperties]
  renderers [
   {rendererClass=&NiagaraSpriteRendererProperties}
   {rendererClass=&NiagaraSpriteRendererProperties, rendererIndex=1}
  ]
 }
 Laser {
  bEnabled=false
  emitterUpdateScript [
   &EmitterState()
   &SpawnBurst_Instantaneous(Spawn Count=10, Spawn Group=1, Use Spawn Probability=false, UseLoopCountLimit=false) {module=SpawnBurst_Instantaneous001}
   &SpawnBurst_Instantaneous(Spawn Count=10, Use Spawn Probability=false, UseLoopCountLimit=false)
  ]
  particleSpawnScript [
   /SetVariables(Particles.VisibilityTag=&SpawnGroup()) {isSetParameters=true}
   &InitializeParticle(
    Color=&MergeHelper_ColorSwitch(
     0=<rgba 0.9219 0.0576 0.0576 0.7>
     1=<rgba 0.9219 0.821 0.821 1>
     2=?
     3=?
     Index=&SpawnGroup()
     Num=::NewEnumerator0
    )
    Lifetime=$System.IntroLength
    Ribbon Width=&MergeHelper_FloatSwitch(0=6, 1=2, 2=?, 3=?, Index=&SpawnGroup(), Num=::NewEnumerator0)
    Ribbon Width Mode=::NewEnumerator1
    Sprite Size Mode=::NewEnumerator3
    Uniform Sprite Size=&Multiply_Float(A=10, B=&FN_FrontEndFloatModifier(Gameplay=0.5, Front End=$User.FrontEndScale))
    UsePositionOffset=false
    Write Material Random=true
   )
   &PartitionParticles(
    Number Of Partitions=1
    Particles Per Partition=10
    Partition Mode:ENiagara_PartitionMode_v2=::NewEnumerator1
    Write Partition Index to Ribbon Link Order:ENiagara_AttributeSamplingApplyOutput=::NewEnumerator0
    Write Partition to Ribbon ID=::NewEnumerator0
   )
  ]
  particleUpdateScript [
   &ParticleState()
   &ScaleColor(
    Scale Alpha=&FloatFromCurve(FloatCurve=DataInterface() {propertyValues {Curve {
        defaultValue=0
        keys [
         {arriveTangent=0.0228, interpMode=RCIM_Cubic, leaveTangent=0.0228, tangentMode=RCTM_User, time=0.1408, value=0.9847}
         {arriveTangent=-0.161, interpMode=RCIM_Cubic, leaveTangent=-0.161, tangentMode=RCTM_User, time=0.5189, value=1}
         {arriveTangent=-0.0405, interpMode=RCIM_Cubic, leaveTangent=-0.0405, tangentMode=RCTM_User, time=0.75, value=0}
        ]
       }}})
    Scale RGB=&MakeVectorFromLinearColorRGB(Color=&ColorFromCurve(ColorCurve=DataInterface() {propertyValues {
        AlphaCurve {keys [<at 0 1 RCIM_Cubic>]}
        BlueCurve {keys [<at 0.3291 1 RCIM_Cubic>,<at 0.8239 1>,<at 1 1 RCIM_Cubic>]}
        GreenCurve {keys [<at 0.3291 0.1287 RCIM_Cubic>,<at 0.8239 0.3778>,<at 1 0.137 RCIM_Cubic>]}
        RedCurve {keys [<at 0.3291 0 RCIM_Cubic>,<at 0.8239 0>,<at 1 0 RCIM_Cubic>]}
       }}))
    ScaleA=true
    ScaleRGB=false
   )
   /SetVariables(Particles.Position=&LerpPosition(
     A=Linked() {linkedVariable {name=System.PortalPosition, type=Vector3f}}
     B=&LerpPosition(
      A=Linked() {linkedVariable {name=System.PortalPosition, type=Vector3f}}
      B=Linked() {linkedVariable {name=System.BurstPosition, type=Vector3f}}
      Alpha=&FloatFromCurve(FloatCurve=DataInterface() {propertyValues {Curve {defaultValue=0, keys [{arriveTangent=2.2701, time=0, value=0}, {leaveTangent=-0.3557, time=0.65, value=1}]}}})
     )
     Alpha=&FloatFromCurve(
      FloatCurve=DataInterface() {propertyValues {Curve {keys [<at 0 1>,<at 1 0>]}}}
      CurveIndex=$Particles.PartitionParticles.Normalized Index In Partition
     )
    )) {isSetParameters=true}
   &self:GimbalLock_01(
    Gimbal Lock Direction:NiagaraPosition=&ReturnCameraVector(Camera Vectors=::NewEnumerator4, Coordinate Space=::Local, Origin Position=$Particles.Position)
    Position:NiagaraPosition=$Particles.Position
   ) {module=GimbalLock_01}
   &ScaleRibbonWidth(Ribbon Width Scale=&FloatFromCurve(FloatCurve=DataInterface() {propertyValues {Curve {
        defaultValue=1
        keys [
         {arriveTangent=-0.0533, interpMode=RCIM_Cubic, leaveTangent=-0.0533, tangentMode=RCTM_User, time=0, value=0}
         {arriveTangent=1.5558, interpMode=RCIM_Cubic, leaveTangent=1.5558, tangentMode=RCTM_User, time=0.4331, value=0.1457}
         {arriveTangent=4.3334, interpMode=RCIM_Cubic, leaveTangent=4.3334, tangentMode=RCTM_User, time=0.65, value=1}
        ]
       }}}))
  ]
  rendererClasses [&NiagaraRibbonRendererProperties]
  renderers [
   {rendererClass=&NiagaraRibbonRendererProperties}
   {rendererClass=&NiagaraRibbonRendererProperties, rendererIndex=1}
  ]
 }
 LaserSprites {
  emitterUpdateScript [
   &EmitterState()
   &SpawnBurst_Instantaneous(
    Spawn Count=&RandomRangeInt(Minimum=6, Maximum=8)
    Spawn Time=0.1
    Use Spawn Probability=false
    UseLoopCountLimit=false
   )
  ]
  particleSpawnScript [
   &InitializeParticle(
    Color=<rgba 3 3 3 0.425>
    Lifetime Max=0.4
    Lifetime Min=0.225
    Lifetime Mode=::NewEnumerator1
    Position Offset=$System.PortalPosition
    Sprite Size Max=<1.2 15>
    Sprite Size Min=<0.4 8>
    Sprite Size Mode=::NewEnumerator2
    UsePositionOffset=true
    Write Material Random=true
   )
   &ShapeLocation(
    Cylinder Height=45
    Cylinder Radius=5
    MaskSpawnGroup=false
    Override Seed=false
    Shape Primitive=::NewEnumerator1
    Surface Only=true
    Surface Only Band Thickness=0.65
    UseOwnerScale=true
   )
  ]
  particleUpdateScript [
   &ParticleState()
   &GravityForce(Gravity=&MakeVector(X=0, Y=0, Z=&RandomRangeFloat(Minimum=-120, Maximum=-200)))
   &Drag(Drag=&RandomRangeFloat(Minimum=0.8, Maximum=1.2), Use Linear Drag=true, Use Rotational Drag=true)
   &SolveForcesAndVelocity(Clamp Velocity=false, Limit Acceleration=false)
   &ScaleColor(
    Scale Alpha=&FloatFromCurve(FloatCurve=DataInterface() {propertyValues {Curve {
        defaultValue=1
        keys [
         {arriveTangent=9.0725, interpMode=RCIM_Cubic, leaveTangent=9.0725, tangentMode=RCTM_User, time=0, value=0}
         {arriveTangent=-0.1011, interpMode=RCIM_Cubic, leaveTangent=-0.1011, tangentMode=RCTM_User, time=0.1408, value=0.9847}
         {arriveTangent=-0.0405, interpMode=RCIM_Cubic, leaveTangent=-0.0405, tangentMode=RCTM_User, time=1, value=0}
        ]
       }}})
    ScaleA=true
    ScaleRGB=true
   )
   &self:GimbalLock_01(
    Gimbal Lock Direction:NiagaraPosition=&ReturnCameraVector(Camera Vectors=::NewEnumerator4, Coordinate Space=::Local, Origin Position=$Particles.Position)
    Position:NiagaraPosition=$Particles.Position
   ) {enabled=false, module=GimbalLock_01}
  ]
  rendererClasses [&NiagaraSpriteRendererProperties]
  renderers [{rendererClass=&NiagaraSpriteRendererProperties}]
 }
 LaserSprites001 {
  emitterUpdateScript [
   &EmitterState(Life Cycle Mode=::NewEnumerator1, Loop Behavior=::NewEnumerator1, Loop Delay=0.14, Loop Duration=0.4, UseLoopDelay=true)
   &SpawnRate(SpawnRate=32, Use Spawn Probability=false)
   &SpawnRate(SpawnGroup=1, SpawnRate=56, Use Spawn Probability=false) {module=SpawnRate001}
  ]
  particleSpawnScript [
   /SetVariables(Particles.VisibilityTag=&SpawnGroup()) {isSetParameters=true}
   &InitializeParticle(
    Color=<rgba 15 11.9665 4.8438 0.425>
    Lifetime Max=0.3
    Lifetime Min=0.125
    Lifetime Mode=::NewEnumerator1
    Position Offset=&AddVectorToPosition(
     Position=Linked() {linkedVariable {name=System.BurstPosition, type=Vector3f}}
     Vector=&FN_FrontEndVector3Modifier(Gameplay=<0 0 12>, Front End=<0 0 -6>)
    )
    Sprite Size Max=<1.2 20>
    Sprite Size Min=<0.4 12>
    Sprite Size Mode=::NewEnumerator2
    UsePositionOffset=true
    Write Material Random=true
   )
   &ShapeLocation(MaskSpawnGroup=false, Override Seed=false, Sphere Radius=18, Sphere Surface Distribution=0.6667, UseOwnerScale=true)
  ]
  particleUpdateScript [
   &ParticleState()
   &PointAttractionForce(
    Attraction Strength=120
    Kill Radius=2
    Kill Within Radius=true
    Module.Use Falloff=true
    Position To Attract:NiagaraPosition=&AddVectorToPosition(
     Position=$Particles.Position
     Vector=&FN_FrontEndVector3Modifier(Gameplay=<0 0 0>, Front End=<0 0 35>)
    )
    UseAttractorOffset=false
   )
   &SolveForcesAndVelocity(Clamp Velocity=false, Limit Acceleration=false)
   &ScaleColor(
    Scale Alpha=&FloatFromCurve(FloatCurve=DataInterface() {propertyValues {Curve {
        defaultValue=1
        keys [
         {arriveTangent=9.0725, interpMode=RCIM_Cubic, leaveTangent=9.0725, tangentMode=RCTM_User, time=0, value=0}
         {arriveTangent=-0.1011, interpMode=RCIM_Cubic, leaveTangent=-0.1011, tangentMode=RCTM_User, time=0.1408, value=0.9847}
         {arriveTangent=-0.0405, interpMode=RCIM_Cubic, leaveTangent=-0.0405, tangentMode=RCTM_User, time=1, value=0}
        ]
       }}})
    ScaleA=true
    ScaleRGB=true
   )
   &self:GimbalLock_01(
    Gimbal Lock Direction:NiagaraPosition=&ReturnCameraVector(Camera Vectors=::NewEnumerator4, Coordinate Space=::Local, Origin Position=$Particles.Position)
    Position:NiagaraPosition=$Particles.Position
   ) {module=GimbalLock_01}
   /SetVariables(Particles.Circle Size:Vector2f=&Vector2DFromFloat(Value=&RandomRangeFloat(Minimum=0.4, Maximum=1.4))) {isSetParameters=true}
   /SetVariables(
    Particles.Glow Color=<rgba 15 11.9665 4.8438 0.01>
    Particles.Glow Size=&Multiply_Vector2D_ByFloat(Float=2.6, Vector2D=$Particles.Circle Size)
   ) {isSetParameters=true}
  ]
  rendererClasses [&NiagaraSpriteRendererProperties]
  renderers [
   {rendererClass=&NiagaraSpriteRendererProperties}
   {rendererClass=&NiagaraSpriteRendererProperties, rendererIndex=1}
   {rendererClass=&NiagaraSpriteRendererProperties, rendererIndex=2}
  ]
 }
 Minimal {
  emitterUpdateScript [
   &EmitterState()
   &SpawnBurst_Instantaneous(Use Spawn Probability=false, UseLoopCountLimit=false)
   &SpawnBurst_Instantaneous(Spawn Group=1, Use Spawn Probability=false, UseLoopCountLimit=false) {module=SpawnBurst_Instantaneous001}
  ]
  particleSpawnScript [
   /SetVariables(Particles.VisibilityTag=&SpawnGroup()) {isSetParameters=true}
   &InitializeParticle(
    Color=&MaskLinearColorBySpawnGroup(Masked Linear Color=<rgba 1 0.0081 0 0.4348>, Passthrough Linear Color=<rgba 15 4.4162 3.75 1>)
    Lifetime=$System.IntroLength
    Sprite Size=&MaskVector2dBySpawnGroup(Masked Vector2D=<20 35>, Passthrough Vector2D=<3.5 27.2805>, Spawn Group Mask=0)
    Sprite Size Mode=::NewEnumerator4
    UsePositionOffset=false
    Write Material Random=true
   )
  ]
  particleUpdateScript [
   &ParticleState()
   &ScaleSpriteSize(
    Non-Uniform Curve Sprite Scale=DataInterface() {propertyValues {
      XCurve {keys [<at 0 0 RCIM_Cubic>,<at 0.2 1 RCIM_Cubic>]}
      YCurve {keys [{tangentMode=RCTM_User, time=0, value=1}, {tangentMode=RCTM_User, time=1, value=1}]}
     }}
    Scale Sprite Size Mode=::NewEnumerator3
   )
   &ScaleColor(
    Scale Alpha=&FloatFromCurve(FloatCurve=DataInterface() {propertyValues {Curve {keys [<at 0.2 1 RCIM_Cubic>, {arriveTangent=-11.3136, interpMode=RCIM_Cubic, leaveTangent=-11.3136, tangentMode=RCTM_User, time=0.45, value=0}]}}})
    ScaleA=true
    ScaleRGB=true
   )
   /SetVariables(Particles.Position=&Lerp_Vector(
     A=$System.PortalPosition
     B=$System.BurstPosition
     Alpha=&FloatFromCurve(FloatCurve=DataInterface() {propertyValues {Curve {keys [<at 0 0>,<at 0.6 1>]}}})
    )) {isSetParameters=true}
   &self:GimbalLock_01(
    Gimbal Lock Direction:NiagaraPosition=&ReturnCameraVector(Camera Vectors=::NewEnumerator4, Coordinate Space=::Local, Origin Position=$Particles.Position)
    Position:NiagaraPosition=$Particles.Position
   ) {module=GimbalLock_01}
  ]
  rendererClasses [&NiagaraSpriteRendererProperties]
  renderers [
   {rendererClass=&NiagaraSpriteRendererProperties}
   {rendererClass=&NiagaraSpriteRendererProperties, rendererIndex=1}
  ]
 }
 Shockwave001 {
  emitterUpdateScript [
   &EmitterState(Life Cycle Mode=::NewEnumerator1, Loop Behavior=::NewEnumerator1, Loop Delay=0.72, Loop Duration=1, UseLoopDelay=true)
   &SpawnBurst_Instantaneous(Use Spawn Probability=false, UseLoopCountLimit=false)
  ]
  particleSpawnScript [&InitializeParticle(
    Color=<rgba 1 1 1 5>
    Lifetime=0.3
    Mesh Scale Mode=::NewEnumerator3
    Mesh Uniform Scale=&Multiply_Float(A=0.07, B=$System.FrontendScale)
    Position Offset=&AddVectorToPosition(
     Position=Linked() {linkedVariable {name=System.BurstPosition, type=Vector3f}}
     Vector=&FN_FrontEndVector3Modifier(Gameplay=<0 0 12>, Front End=<0 0 -6>)
    )
    Sprite Size Mode=::NewEnumerator3
    Uniform Sprite Size=60
    UsePositionOffset=true
    Write Material Random=true
   )]
  particleUpdateScript [
   &ParticleState()
   &DynamicMaterialParameters(
    Index 0 Param 1=1
    Index 0 Param 2=2
    Module.Index0Float0Write=true
    Module.Index0Float1Write=true
    Module.Index0Float2Write=true
    Module.Index0Float3Write=true
   )
   &ScaleColor(
    Scale Alpha=&FloatFromCurve(
     FloatCurve=DataInterface() {propertyValues {Curve {keys [
         {arriveTangent=-1.5846, interpMode=RCIM_Cubic, leaveTangent=-1.5846, tangentMode=RCTM_User, time=0, value=1}
         {arriveTangent=-0.3296, tangentMode=RCTM_User, time=1, value=0}
        ]}}}
     Scale Curve=&Multiply_Float(A=1, B=$System.TODScalar)
    )
    Scale RGB=&MakeVectorFromLinearColorRGB(Color=&ColorFromCurve(ColorCurve=DataInterface() {propertyValues {
        AlphaCurve {keys [<at 0 1.5>,<at 1 0>]}
        BlueCurve {keys [<at 0.0389 0.0793>,<at 1 0.0275>]}
        GreenCurve {keys [<at 0.0389 0.2135>,<at 1 0.0944>]}
        RedCurve {keys [<at 0.0389 1>,<at 1 1>]}
       }}))
    ScaleA=true
    ScaleRGB=true
   )
   &ScaleMeshSize(Scale Factor=&VectorFromFloat(Value=&FloatFromCurve(FloatCurve=DataInterface() {propertyValues {Curve {keys [
          {arriveTangent=2.4306, interpMode=RCIM_Cubic, leaveTangent=2.4306, tangentMode=RCTM_User, time=0, value=0}
          {arriveTangent=0.046, interpMode=RCIM_Cubic, leaveTangent=0.046, tangentMode=RCTM_User, time=1, value=1}
         ]}}})))
   &CameraOffset(Camera Offset Amount=15)
   /SetVariables(
    Particles.SecondaryDyn:Vector4f=&MakeVector4(X=2.5, Y=3, Z=0, W=0)
    Particles.SecondarySphereColor:LinearColor=&MakeLinearColorFromVectorAndFloat(
     "Vector (RGB)"=&MakeVectorFromLinearColorRGB(Color=&ColorFromCurve(ColorCurve=DataInterface() {propertyValues {
         AlphaCurve {keys [<at 0 1.5>,<at 1 0>]}
         BlueCurve {keys [<at 0.0389 0.577>,<at 1 0.0698>]}
         GreenCurve {keys [<at 0.0389 1.3452>,<at 1 0.0889>]}
         RedCurve {keys [<at 0.0389 5>,<at 1 1>]}
        }}))
     "Float (Alpha)"=&FloatFromCurve(
      FloatCurve=DataInterface() {propertyValues {Curve {keys [
          {arriveTangent=-1.5846, interpMode=RCIM_Cubic, leaveTangent=-1.5846, tangentMode=RCTM_User, time=0, value=1}
          {arriveTangent=-0.3296, tangentMode=RCTM_User, time=1, value=0}
         ]}}}
      Scale Curve=&Multiply_Float(A=1, B=$System.TODScalar)
     )
    )
    Particles.ShockwaveColor:LinearColor=&MakeLinearColorFromVectorAndFloat(
     "Vector (RGB)"=&MakeVectorFromLinearColorRGB(Color=<rgba 1 0.1802 0.0268 1>)
     "Float (Alpha)"=&FloatFromCurve(
      FloatCurve=DataInterface() {propertyValues {Curve {keys [
          {arriveTangent=0.0068, interpMode=RCIM_Cubic, leaveTangent=0.0068, tangentMode=RCTM_User, time=0, value=0}
          {arriveTangent=-0.018, interpMode=RCIM_Cubic, leaveTangent=-0.018, tangentMode=RCTM_User, time=0.1, value=1}
          {arriveTangent=0.0005, interpMode=RCIM_Cubic, leaveTangent=0.0005, tangentMode=RCTM_User, time=0.7, value=0}
         ]}}}
      Scale Curve=&Multiply_Float(A=0.5, B=$System.TODScalar)
     )
    )
    Particles.ShockwaveScale:Vector3f=&Multiply_VectorByFloat(Vector=$Particles.Scale, Float=-3)
   ) {isSetParameters=true}
   /SetVariables(Particles.Flare Color:LinearColor=&MultiplyLinearColor(Linear Color=$Particles.Color, Scale Factor=<1 1 1 0.425>)) {isSetParameters=true}
  ]
  rendererClasses [&NiagaraMeshRendererProperties, &NiagaraSpriteRendererProperties]
  renderers [
   {rendererClass=&NiagaraMeshRendererProperties, rendererIndex=0}
   {rendererClass=&NiagaraMeshRendererProperties, rendererIndex=1}
   {rendererClass=&NiagaraMeshRendererProperties, rendererIndex=2}
   {rendererClass=&NiagaraSpriteRendererProperties, rendererIndex=3}
  ]
 }
}
summary {
 systemName=NS_ReactFX_CacaoLynx
 emitters [
  {emitterName=Heart, rendererClasses [&NiagaraSpriteRendererProperties]}
  {bEnabled=false, emitterName=Laser, rendererClasses [&NiagaraRibbonRendererProperties]}
  {emitterName=LaserSprites, rendererClasses [&NiagaraSpriteRendererProperties]}
  {emitterName=LaserSprites001, rendererClasses [&NiagaraSpriteRendererProperties]}
  {emitterName=Shockwave001, rendererClasses [&NiagaraMeshRendererProperties, &NiagaraSpriteRendererProperties]}
  {emitterName=Fire002, rendererClasses [&NiagaraSpriteRendererProperties]}
  {emitterName=Empty, rendererClasses []}
  {emitterName=Empty001, rendererClasses []}
  {emitterName=Minimal, rendererClasses [&NiagaraSpriteRendererProperties]}
 ]
 userVariables [
  User.bIsFrontEnd {type=NiagaraBool}
  User.bIsFrontEndPreview {type=NiagaraBool}
  User.FrontEndScale {type=NiagaraFloat, defaultValue=0.4, description=""}
 ]
}
