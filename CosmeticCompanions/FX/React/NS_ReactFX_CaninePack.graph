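# Frost-themed Niagara reaction effect for the CaninePack cosmetic companion: a heart appears, pops and plays a shared petting-interaction sound.
# System values for intro length, start and burst positions, and time-of-day tint drive all visible emitters.
# A glowing heart with spiraling ribbon trail, smoke and sparkles builds up, then bursts into snow, smoke and nine SubUV shards flung in a camera-facing ring.
# User.FrontEndScale resizes it for front-end display.
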NiagaraSystem /CosmeticCompanions/FX/React/NS_ReactFX_CaninePack
dependencies {
 usedDataInterfaces [
  &NiagaraDataInterfaceAudioPlayer
  &NiagaraDataInterfaceCurve
  &NiagaraDataInterfaceColorCurve
  &NiagaraDataInterfaceSpriteRendererInfo
  &NiagaraDataInterfaceCamera
 ]
 usedDynamicInputs [
  &FN_FrontEndVector3Modifier
  &MakeCustomLinearColorFromBool
  &Boolean_OrOperation
  &MakeFloatFromVector
 ]
 usedModules [
  &self:SystemSpawnScript.SystemScriptSource.SystemScriptGraph.NiagaraNodeAssignment_0.SetVariables_#1
  &self:SystemSpawnScript.SystemScriptSource.SystemScriptGraph.NiagaraNodeAssignment_6.SetVariables_#2
  &self:SystemSpawnScript.SystemScriptSource.SystemScriptGraph.NiagaraNodeAssignment_7.SetVariables_#3
  &self:SystemSpawnScript.SystemScriptSource.SystemScriptGraph.NiagaraNodeAssignment_5.SetVariables_#4
  &self:SystemSpawnScript.SystemScriptSource.SystemScriptGraph.NiagaraNodeAssignment_3.SetVariables_#5
  &self:SystemSpawnScript.SystemScriptSource.SystemScriptGraph.NiagaraNodeAssignment_1.SetVariables_#6
  &SystemState
  &self:SystemSpawnScript.SystemScriptSource.SystemScriptGraph.NiagaraNodeAssignment_8.SetVariables_#7
 ]
}
emitters {
 Frosty_Smoke {
  emitterUpdateScript [
   &EmitterState()
   &SpawnRate(
    SpawnRate=&FloatFromCurve(
     FloatCurve=DataInterface() {propertyValues {Curve {keys [{arriveTangent=3.0606, interpMode=RCIM_Constant, time=0, value=0}, <at 0.1 1 RCIM_Constant>, <at 0.5 0 RCIM_Constant>]}}}
     CurveIndex=&Divide_Float(A=$Emitter.Age, B=$System.IntroLength)
     Scale Curve=10
    )
    Use Spawn Probability=false
   )
   &SpawnBurst_Instantaneous(
    Spawn Count=5
    Spawn Group=1
    Spawn Time=&Add_Float(A=$System.IntroLength, B=-0.2)
    Use Spawn Probability=false
    UseLoopCountLimit=false
   )
  ]
  particleSpawnScript [
   &InitializeParticle(
    Color=&MultiplyLinearColor(
     Linear Color=&RandomRangeLinearColor(Minimum=<rgba 0.3906 0.7302 1 0.4>, Maximum=<rgba 0.5469 0.7994 1 0.4>)
     Scale Factor=&MakeVector4_FromLinearColor(LinearColor=$System.TOD)
    )
    Lifetime=&RandomRangeFloat(Minimum=0.8, Maximum=1.5)
    Sprite Rotation Mode=::NewEnumerator2
    Sprite Size Mode=::NewEnumerator3
    Uniform Sprite Size=&Multiply_Float(A=30, B=&FN_FrontEndFloatModifier(Gameplay=1, Front End=$User.FrontEndScale))
    UsePositionOffset=false
    Write Material Random=true
   ) {module=InitializeParticle001}
   &CameraOffset(Camera Offset Amount=20)
   /SetVariables(Particles.Position=&Lerp_Vector(
     A=$System.HeartIntroSpawnPosition
     B=$System.BurstPosition
     Alpha=&FloatFromCurve(
      FloatCurve=DataInterface() {propertyValues {Curve {keys [{arriveTangent=2.5323, interpMode=RCIM_Cubic, leaveTangent=2.5323, tangentMode=RCTM_User, time=0, value=0}, <at 1 1 RCIM_Cubic>]}}}
      CurveIndex=&Add_Float(A=0.1, B=&Divide_Float(A=$Engine.System.Age, B=$System.IntroLength))
     )
    )) {isSetParameters=true, module=SetVariables_#8}
   &AddVelocity(
    Velocity=&RandomRangeVector(Minimum=<-1 -1 -1>, Maximum=<1 1 1>)
    Velocity Speed Scale=&Multiply_Float(
     A=&MergeHelper_FloatSwitch(0=1, 1=50, 2=?, 3=?, Index=&SpawnGroup(), Num=::NewEnumerator0)
     B=&FN_FrontEndFloatModifier(Gameplay=1, Front End=$User.FrontEndScale)
    )
   )
   /SetVariables(Particles.StartIndex:NiagaraInt32=&RandomRangeInt(Minimum=2, Maximum=20)) {isSetParameters=true}
  ]
  particleUpdateScript [
   &ParticleState()
   &SubUVAnimation(
    End Frame:NiagaraInt32=&Add_Integer(A=$Particles.StartIndex, B=&RandomRangeInt(Minimum=20, Maximum=30))
    Start Frame:NiagaraInt32=$Particles.StartIndex
   )
   &GravityForce(Gravity=&RandomRangeVector(Minimum=<0 0 -6>, Maximum=<0 0 -10>))
   &Drag(Drag=&RandomRangeFloat(Minimum=2, Maximum=3), Use Linear Drag=true, Use Rotational Drag=true)
   &SolveForcesAndVelocity(Clamp Velocity=false, Limit Acceleration=false)
   &ScaleColor(
    Scale Alpha=&FloatFromCurve(
     FloatCurve=DataInterface() {propertyValues {Curve {keys [
         <at 0 0 RCIM_Cubic>
         {arriveTangent=0.317, interpMode=RCIM_Cubic, leaveTangent=0.317, tangentMode=RCTM_User, time=0.1, value=0.4}
         <at 1 0 RCIM_Cubic>
        ]}}}
     Scale Curve=&Multiply_Float(
      A=&Multiply_Float(
       A=&RandomRangeFloat(Minimum=0.3, Maximum=0.5)
       B=&FN_FrontEndFloatModifier(Gameplay=1, Front End=0.5)
      )
      B=&FN_Select_Float_from_Time_Of_Day(Morning=0.5, Day=1, Evening=0.5, Night=0.5)
     )
    )
    Scale RGB=<1 1 1.1>
    ScaleA=true
    ScaleRGB=true
   )
   &DynamicMaterialParameters(
    Index 0 Param 4=1
    Module.Param0WriteEnabled=true
    Module.Param1WriteEnabled=true
    Module.Param2WriteEnabled=true
    Module.Param3WriteEnabled=true
   )
   &ScaleSpriteSize(Scale Factor=&Vector2DFromFloat(Value=&FloatFromCurve(FloatCurve=DataInterface() {propertyValues {Curve {keys [
          {arriveTangent=2.1241, interpMode=RCIM_Cubic, leaveTangent=2.1241, tangentMode=RCTM_User, time=0, value=0.5}
          {arriveTangent=0.8108, interpMode=RCIM_Cubic, leaveTangent=0.8108, tangentMode=RCTM_User, time=1, value=1.5}
         ]}}})))
  ]
  rendererClasses [&NiagaraSpriteRendererProperties]
  renderers [{rendererClass=&NiagaraSpriteRendererProperties}]
 }
 Heart {
  emitterSpawnScript [
   &InitializeAudioPlayer(Audio Player=DataInterface() {propertyValues {
      Attenuation=null
      Concurrency=null
      MaxPlaysPerTick=10
      SoundToPlay=/CosmeticCompanions/Assets/Audio_Shared/Interaction/Petting/SC_Mimosa_Interaction_React_Default
      bAllowLoopingOneShotSounds=false
      bLimitPlaysPerTick=true
      bOnlyActiveDuringGameplay=false
      bStopWhenComponentIsDestroyed=true
      ConfigurationUserParameter {parameter {name=None, typeDefHandle {registeredTypeIndex=1}, varData []}}
      ParameterNames []
     }})
   /Niagara/Modules/Audio/PlayAudio() {module=PlayAudio001}
  ]
  emitterUpdateScript [
   &EmitterState()
   &SpawnRate(
    SpawnGroup=1
    SpawnRate=&FloatFromCurve(FloatCurve=DataInterface() {propertyValues {Curve {keys [<at 0.49 1>,<at 0.5 0>]}}}, CurveIndex=$Emitter.NormalizedLoopAge, Scale Curve=50)
    Use Spawn Probability=false
   )
   &SpawnBurst_Instantaneous(Use Spawn Probability=false, UseLoopCountLimit=false)
  ]
  particleSpawnScript [
   /SetVariables(Particles.VisibilityTag=$Emitter.SpawnGroup) {isSetParameters=true}
   &InitializeParticle(
    Color=$System.TOD
    Lifetime=$System.IntroLength
    Sprite Rotation Mode=::NewEnumerator0
    Sprite Size Mode=::NewEnumerator3
    Uniform Sprite Size=&Multiply_Float(A=18, B=&FN_FrontEndFloatModifier(Gameplay=1, Front End=$User.FrontEndScale))
    UsePositionOffset=false
    Write Material Random=true
   )
  ]
  particleUpdateScript [
   &ParticleState()
   &ScaleSpriteSize(
    Scale Sprite Size Mode=::NewEnumerator0
    Uniform Scale Factor=&Add_Float(
     A=0
     B=&FloatFromCurve(FloatCurve=DataInterface() {propertyValues {Curve {
         defaultValue=0
         keys [
          <at 0 0.3 RCIM_Cubic>
          {arriveTangent=1.25, leaveTangent=1.25, time=0.4, value=1}
          {arriveTangent=6.566, interpMode=RCIM_Cubic, leaveTangent=6.566, tangentMode=RCTM_User, time=0.9, value=1.1}
          {arriveTangent=35.933, interpMode=RCIM_Cubic, leaveTangent=35.933, tangentMode=RCTM_User, time=1, value=3}
         ]
        }}})
    )
   ) {module=ScaleSpriteSize002}
   &ScaleColor(
    Scale Alpha=&FloatFromCurve(FloatCurve=DataInterface() {propertyValues {Curve {
        defaultValue=0
        keys [
         {arriveTangent=9.0725, interpMode=RCIM_Cubic, leaveTangent=9.0725, tangentMode=RCTM_User, time=0, value=0}
         {arriveTangent=-0.2441, leaveTangent=-0.2441, time=0.1, value=0.9847}
         {arriveTangent=-1.146, leaveTangent=-1.146, time=0.98, value=1}
         {arriveTangent=-0.0405, interpMode=RCIM_Cubic, leaveTangent=-0.0405, tangentMode=RCTM_User, time=1, value=0}
        ]
       }}})
    ScaleA=true
    ScaleRGB=true
   )
   /SetVariables(Particles.Position=$System.OffsetPosition) {isSetParameters=true}
   &DynamicMaterialParameters(
    Index 0 Param 1=&FloatFromCurve(FloatCurve=DataInterface() {propertyValues {Curve {keys [{arriveTangent=0.6296, interpMode=RCIM_Cubic, leaveTangent=0.6296, tangentMode=RCTM_User, time=0.5, value=1}, <at 1 2>]}}})
    Index 0 Param 2=&FloatFromCurve(FloatCurve=DataInterface() {propertyValues {Curve {keys [{arriveTangent=0.6296, interpMode=RCIM_Cubic, leaveTangent=0.6296, tangentMode=RCTM_User, time=0.5, value=1}, <at 0.6 50>, <at 0.7 1>]}}})
    Index 0 Param 3=1
    Module.Index0Float0Write=true
    Module.Index0Float1Write=true
    Module.Index0Float2Write=true
    Module.Index0Float3Write=true
   )
  ]
  rendererClasses [&NiagaraSpriteRendererProperties]
  renderers [{rendererClass=&NiagaraSpriteRendererProperties}]
 }
 HeartGlow {
  emitterUpdateScript [
   &EmitterState()
   &SpawnBurst_Instantaneous(Use Spawn Probability=false, UseLoopCountLimit=false)
  ]
  particleSpawnScript [&InitializeParticle(
    Color=$System.TOD
    Position Offset=$System.BurstPosition
    Sprite Size Mode=::NewEnumerator3
    Uniform Sprite Size=&Multiply_Float(A=25, B=&FN_FrontEndFloatModifier(Gameplay=1, Front End=$User.FrontEndScale))
    UsePositionOffset=true
    Write Material Random=true
   )]
  particleUpdateScript [
   &ParticleState()
   &ScaleSpriteSize(
    Scale Sprite Size Mode=::NewEnumerator0
    Uniform Scale Factor=&FloatFromCurve(
     FloatCurve=DataInterface() {propertyValues {Curve {keys [
         {arriveTangent=3.3458, interpMode=RCIM_Cubic, leaveTangent=3.3458, tangentMode=RCTM_User, time=0, value=1}
         {arriveTangent=1.251, tangentMode=RCTM_User, time=0.7, value=3}
         <at 0.9 2>
         <at 1.0056 5>
        ]}}}
     Scale Curve=&Waveform("Waveform [1]"=::NewEnumerator12, "[1] Frequency"=5, "[1] Amplitude Min/Max"=<0.7 1>)
    )
   ) {module=ScaleSpriteSize002}
   &ScaleColor(
    Scale Alpha=&FloatFromCurve(FloatCurve=DataInterface() {propertyValues {Curve {
        defaultValue=0
        keys [
         {arriveTangent=0.6582, interpMode=RCIM_Cubic, leaveTangent=0.6582, tangentMode=RCTM_User, time=0.3, value=0}
         {arriveTangent=1.6667, interpMode=RCIM_Cubic, leaveTangent=1.6667, time=0.7134, value=0.3}
         {arriveTangent=0.3137, interpMode=RCIM_Cubic, leaveTangent=0.3137, tangentMode=RCTM_User, time=0.9, value=1}
        ]
       }}})
    Scale RGB=&MakeVectorFromLinearColorRGB(Color=&ColorFromCurve(ColorCurve=DataInterface() {propertyValues {
        AlphaCurve {keys [<at 0 1 RCIM_Cubic>]}
        BlueCurve {keys [<at 0.3291 1 RCIM_Cubic>,<at 0.8239 1>,<at 1 1 RCIM_Cubic>]}
        GreenCurve {keys [<at 0.3291 0.1287 RCIM_Cubic>,<at 0.8239 0.3778>,<at 1 0.137 RCIM_Cubic>]}
        RedCurve {keys [<at 0.3291 0 RCIM_Cubic>,<at 0.8239 0>,<at 1 0 RCIM_Cubic>]}
       }}))
    ScaleA=true
    ScaleRGB=true
   )
   /SetVariables(Particles.Position=$System.OffsetPosition) {isSetParameters=true}
   &DynamicMaterialParameters(
    Index 0 Param 1=1
    Index 0 Param 3=1
    Module.Index0Float0Write=true
    Module.Index0Float1Write=true
    Module.Index0Float2Write=true
    Module.Index0Float3Write=true
   )
   &FN_Sprite_Attributes(
    Color=&MakeLinearColorFromVectorAndFloat(
     "Vector (RGB)"=&MakeVectorFromLinearColorRGB(Color=<rgba 0.5149 1 0.9911 1>)
     "Float (Alpha)"=&FloatFromCurve(
      FloatCurve=DataInterface() {propertyValues {Curve {
         defaultValue=0
         keys [
          {arriveTangent=0.6582, interpMode=RCIM_Cubic, leaveTangent=0.6582, tangentMode=RCTM_User, time=0, value=0}
          {arriveTangent=0.3137, interpMode=RCIM_Cubic, leaveTangent=0.3137, tangentMode=RCTM_User, time=0.8, value=1}
          <at 0.9832 0 RCIM_Cubic>
         ]
        }}}
      Scale Curve=0.2
     )
    )
    Create Color Attribute=true
    Create SpriteSize Attribute=true
    SpriteSize=&Vector2DFromFloat(Value=&Multiply_Float(
      A=&Multiply_Float(A=20, B=&FN_FrontEndFloatModifier(Gameplay=1, Front End=$User.FrontEndScale))
      B=&FloatFromCurve(FloatCurve=DataInterface() {propertyValues {Curve {keys [
           {arriveTangent=3.3458, interpMode=RCIM_Cubic, leaveTangent=3.3458, tangentMode=RCTM_User, time=0, value=1}
           {arriveTangent=1.251, tangentMode=RCTM_User, time=0.5, value=3}
           <at 0.8 3>
           <at 0.9809 5>
          ]}}})
     ))
    Write to Module Namespace=true
   )
  ]
  rendererClasses [&NiagaraSpriteRendererProperties]
  renderers [
   {rendererClass=&NiagaraSpriteRendererProperties}
   {rendererClass=&NiagaraSpriteRendererProperties, rendererIndex=1}
  ]
 }
 LWE_AirGlow {rendererClasses [], renderers []}
 LWE_Corona {rendererClasses [], renderers []}
 LWE_Mist {rendererClasses [], renderers []}
 LWE_Rays {rendererClasses [], renderers []}
 LWE_Rays001 {rendererClasses [], renderers []}
 Pop_Shatter {
  emitterUpdateScript [
   &EmitterState()
   &SpawnBurst_Instantaneous(
    Spawn Count=9
    Spawn Group=1
    Spawn Time=&Add_Float(A=$System.IntroLength, B=-0.1)
    Use Spawn Probability=false
    UseLoopCountLimit=false
   )
  ]
  particleSpawnScript [
   &PartitionParticles(Number Of Partitions=9)
   &InitializeParticle(
    Color=$System.TOD
    Lifetime=&RandomRangeFloat(Minimum=0.4, Maximum=0.6)
    Sprite Size Mode=::NewEnumerator3
    Uniform Sprite Size=&Multiply_Float(A=16, B=&FN_FrontEndFloatModifier(Gameplay=1, Front End=$User.FrontEndScale))
    UsePositionOffset=false
    Write Material Random=true
   ) {module=InitializeParticle001}
   &SubUVAnimation(
    End Frame Range Override=9
    "Sprite Renderer "=DataInterface() {propertyValues {SpriteRenderer=&self:Pop_Shatter_0.NiagaraSpriteRendererProperties_0}}
    Start Frame Offset:NiagaraInt32=$Particles.PartitionParticles.Partition
    UseEndFrame=true
    UseStartFrame=false
   )
   /SetVariables(Particles.Position=&AddVectorToPosition(
     Position=Linked() {linkedVariable {name=System.BurstPosition, type=Vector3f}}
     Vector=&Multiply_VectorByFloat(
      Vector=&RotateVectorByQuaternion(
       Vector To Rotate=<0 0 1>
       Quaternion=&MakeQuaternion(
        Quaternion From=0
        Axis=&NormalizeVector(
         Vector To Normalize=&Subtract_Position(
          A=&ReturnCameraPosition(
           Coordinate Space:ENiagaraCoordinateSpace=::Simulation
           Camera Query:NiagaraDataInterfaceCamera=DataInterface() {propertyValues {bRequireCurrentFrameData=false}}
          )
          B=/Niagara/DynamicInputs/Helpers/SimulationPosition()
         )
         Safe Normalize=true
         Safe Fallback Threshold=0
         Safe Fallback Vector=<0 0 0.001>
        )
        Angle=&Subtract_Float(A=&Multiply_FloatByInt(Float=40, Integer=$Particles.PartitionParticles.Partition), B=30)
        From=?
        To=?
        Angle Type=::NewEnumerator0
        Normalize Vectors=false
       )
      )
      Float=5
     )
    )) {isSetParameters=true, module=SetVariables_#8}
   &AddVelocity(
    Velocity=&RotateVectorByQuaternion(
     Vector To Rotate=<0 0 1>
     Quaternion=&MakeQuaternion(
      Quaternion From=0
      Axis=&NormalizeVector(
       Vector To Normalize=&Subtract_Position(
        A=&ReturnCameraPosition(
         Coordinate Space:ENiagaraCoordinateSpace=::Simulation
         Camera Query:NiagaraDataInterfaceCamera=DataInterface() {propertyValues {bRequireCurrentFrameData=false}}
        )
        B=&SimulationPosition()
       )
       Safe Normalize=true
       Safe Fallback Threshold=0
       Safe Fallback Vector=<0 0 0.001>
      )
      Angle=&Subtract_Float(A=&Multiply_FloatByInt(Float=40, Integer=$Particles.PartitionParticles.Partition), B=30)
      From=?
      To=?
      Angle Type=::NewEnumerator0
      Normalize Vectors=false
     )
    )
    Velocity Speed Scale=&RandomRangeFloat(Minimum=50, Maximum=100)
   )
  ]
  particleUpdateScript [
   &ParticleState()
   &Drag(Drag=&RandomRangeFloat(Minimum=4, Maximum=8), Use Linear Drag=true, Use Rotational Drag=true)
   &SolveForcesAndVelocity(Clamp Velocity=false, Limit Acceleration=false)
   &ScaleColor(
    Scale Alpha=&FloatFromCurve(FloatCurve=DataInterface() {propertyValues {Curve {keys [{arriveTangent=0.317, interpMode=RCIM_Cubic, leaveTangent=0.317, tangentMode=RCTM_User, time=0.5, value=1}, <at 1 0 RCIM_Cubic>]}}})
    Scale RGB=&MakeVectorFromLinearColorRGB(Color=&ColorFromCurve(ColorCurve=DataInterface() {propertyValues {
        AlphaCurve {keys [<at 0 1 RCIM_Cubic>,<at 1 0 RCIM_Cubic>]}
        BlueCurve {keys [
          <at 0.0723 10>
          {arriveTangent=-9.9859, interpMode=RCIM_Cubic, leaveTangent=-9.9859, time=0.4126, value=5}
          <at 0.8118 1.0773>
          <at 0.9735 1 RCIM_Cubic>
         ]}
        GreenCurve {keys [
          <at 0.0723 10>
          {arriveTangent=-9.9859, interpMode=RCIM_Cubic, leaveTangent=-9.9859, time=0.4126, value=5}
          <at 0.8118 1.0773>
          <at 0.9735 1 RCIM_Cubic>
         ]}
        RedCurve {keys [
          <at 0.0723 10>
          {arriveTangent=-9.9859, interpMode=RCIM_Cubic, leaveTangent=-9.9859, time=0.4126, value=5}
          <at 0.8118 1.0773>
          <at 0.9735 1 RCIM_Cubic>
         ]}
       }}))
    ScaleA=true
    ScaleRGB=true
   )
   &ScaleSpriteSize(Scale Factor=&Vector2DFromFloat(Value=&FloatFromCurve(FloatCurve=DataInterface() {propertyValues {Curve {keys [<at 0 0.5 RCIM_Cubic>, {arriveTangent=2.1241, interpMode=RCIM_Cubic, leaveTangent=2.1241, tangentMode=RCTM_User, time=0.1, value=1}]}}})))
  ]
  rendererClasses [&NiagaraSpriteRendererProperties]
  renderers [{rendererClass=&NiagaraSpriteRendererProperties}]
 }
 Sparkles_And_Snow {
  emitterUpdateScript [
   &EmitterState()
   &SpawnRate(
    SpawnRate=&FloatFromCurve(
     FloatCurve=DataInterface() {propertyValues {Curve {keys [{arriveTangent=3.0606, interpMode=RCIM_Constant, time=0, value=0}, <at 0.1 1 RCIM_Constant>, <at 0.5 0 RCIM_Constant>]}}}
     CurveIndex=&Divide_Float(A=$Emitter.Age, B=$System.IntroLength)
     Scale Curve=25
    )
    Use Spawn Probability=false
   ) {module=SpawnRate001}
   &SpawnBurst_Instantaneous(
    Spawn Count=&RandomRangeInt(Minimum=10, Maximum=15)
    Spawn Group=1
    Spawn Time=&Add_Float(A=$System.IntroLength, B=-0.1)
    Use Spawn Probability=false
    UseLoopCountLimit=false
   )
  ]
  particleSpawnScript [
   /SetVariables(Particles.VisibilityTag=&MakeIntFromFloat(
     Float=&FloatFromCurve(
      FloatCurve=DataInterface() {propertyValues {Curve {keys [<at 0 0 RCIM_Constant>,<at 0.75 1 RCIM_Constant>,<at 1 1 RCIM_Constant>]}}}
      CurveIndex=&RandomRangeFloat(Minimum=0, Maximum=1)
     )
     Conversion Mode=::NewEnumerator1
    )) {isSetParameters=true}
   &InitializeParticle(
    Color=&MultiplyLinearColor(
     Linear Color=&RandomRangeLinearColor(Minimum=<rgba 0.7309 0.9828 1 1>, Maximum=<rgba 0.3427 0.719 1 1>)
     Scale Factor=&MakeVector4_FromLinearColor(LinearColor=$System.TOD)
    )
    Lifetime=&MergeHelper_FloatSwitch(
     0=&MergeHelper_FloatSwitch(
      0=&RandomRangeFloat(Minimum=0.6, Maximum=1.2)
      1=&RandomRangeFloat(Minimum=0.6, Maximum=1.2)
      2=?
      3=?
      Index=&SpawnGroup()
      Num=::NewEnumerator0
     )
     1=&RandomRangeFloat(Minimum=0.5, Maximum=1)
     2=?
     3=?
     Index=$Particles.VisibilityTag
     Num=::NewEnumerator0
    )
    Sprite Rotation Angle=&MergeHelper_FloatSwitch(
     0=0
     1=&RandomRangeFloat(Minimum=0, Maximum=10)
     2=?
     3=?
     Index=$Particles.VisibilityTag
     Num=::NewEnumerator0
    )
    Sprite Rotation Mode=::NewEnumerator0
    Sprite Size Mode=::NewEnumerator3
    Uniform Sprite Size=&Multiply_Float(
     A=&RandomRangeFloat(Minimum=10, Maximum=12)
     B=&FN_FrontEndFloatModifier(Gameplay=1, Front End=$User.FrontEndScale)
    )
    UsePositionOffset=false
    Write Material Random=true
   )
   /SetVariables(Particles.InitialSize:Vector2f=&Vector2DFromFloat(Value=&Multiply_Float(
      A=&RandomRangeFloat(Minimum=50, Maximum=80)
      B=&FN_FrontEndFloatModifier(Gameplay=1, Front End=$User.FrontEndScale)
     ))) {isSetParameters=true}
   &CameraOffset(Camera Offset Amount=20)
   /SetVariables(Particles.Position=$System.OffsetPosition) {isSetParameters=true}
   &ShapeLocation(
    MaskSpawnGroup=false
    Override Seed=false
    Sphere Radius=&Multiply_Float(
     A=&MergeHelper_FloatSwitch(
      0=10
      1=&RandomRangeFloat(Minimum=10, Maximum=30)
      2=?
      3=?
      Index=&SpawnGroup()
      Num=::NewEnumerator0
     )
     B=&FN_FrontEndFloatModifier(Gameplay=1, Front End=$User.FrontEndScale)
    )
    UseOwnerScale=true
   )
   &AddVelocity(
    Module.Use Falloff Distance=false
    Override Seed=false
    Velocity Mode=::NewEnumerator1
    Velocity Speed=&MergeHelper_FloatSwitch(
     0=5
     1=&RandomRangeFloat(Minimum=10, Maximum=50)
     2=?
     3=?
     Index=&SpawnGroup()
     Num=::NewEnumerator0
    )
   )
  ]
  particleUpdateScript [
   &ParticleState()
   &Drag(Drag=&RandomRangeFloat(Minimum=3, Maximum=5), Use Linear Drag=true, Use Rotational Drag=false)
   &GravityForce(Gravity=&MergeHelper_VectorSwitch(
     0=<0 0 0>
     1=&RandomRangeVector(Minimum=<0 0 -30>, Maximum=<0 0 -50>)
     2=?
     3=?
     Index=$Particles.Initial.VisibilityTag
     Num=::NewEnumerator0
    ))
   &SolveForcesAndVelocity(Clamp Velocity=false, Limit Acceleration=false)
   &ScaleSpriteSize(Scale Factor=&Vector2DFromFloat(Value=&Multiply_Float(
      A=&Subtract_Float(
       A=1
       B=&FloatFromCurve(FloatCurve=DataInterface() {propertyValues {Curve {keys [
            <at 0 0 RCIM_Cubic>
            {arriveTangent=-0.0175, interpMode=RCIM_Cubic, leaveTangent=-0.0175, tangentMode=RCTM_User, time=0.1563, value=0.907}
            {arriveTangent=-0.0042, interpMode=RCIM_Cubic, leaveTangent=-0.0042, tangentMode=RCTM_User, time=0.2646, value=0.0384}
            {arriveTangent=0.0081, interpMode=RCIM_Cubic, leaveTangent=0.0081, tangentMode=RCTM_User, time=0.4692, value=0.7123}
            {arriveTangent=0.034, interpMode=RCIM_Cubic, leaveTangent=0.034, tangentMode=RCTM_User, time=0.6406, value=0.0348}
            {arriveTangent=0.019, interpMode=RCIM_Cubic, leaveTangent=0.019, tangentMode=RCTM_User, time=0.7993, value=0.5053}
            <at 1 0 RCIM_Cubic>
           ]}}})
      )
      B=&FloatFromCurve(
       FloatCurve=DataInterface() {propertyValues {Curve {keys [
           {arriveTangent=10.5722, interpMode=RCIM_Cubic, leaveTangent=10.5722, tangentMode=RCTM_User, time=0, value=0}
           {arriveTangent=-0.0042, interpMode=RCIM_Cubic, leaveTangent=-0.0042, tangentMode=RCTM_User, time=0.1, value=1}
           {arriveTangent=-1.1111, interpMode=RCIM_Cubic, leaveTangent=-1.1111, time=0.5, value=1}
           {arriveTangent=-0.3666, interpMode=RCIM_Cubic, leaveTangent=-0.3666, tangentMode=RCTM_User, time=1, value=0}
          ]}}}
       Scale Curve=2
      )
     ))) {module=ScaleSpriteSize001}
   /SetVariables(Particles.Color2=&MultiplyLinearColor(
     Linear Color=&ColorFromCurve(ColorCurve=DataInterface() {propertyValues {
        AlphaCurve {keys [<at 0 0>,<at 0.15 0.1 RCIM_Cubic>,<at 1 0 RCIM_Cubic>]}
        BlueCurve {keys [<at 0.1483 1 RCIM_Cubic>,<at 0.5161 1 RCIM_Cubic>]}
        GreenCurve {keys [<at 0.1483 0.9451 RCIM_Cubic>,<at 0.5161 0.5037 RCIM_Cubic>]}
        RedCurve {keys [<at 0.1483 0.7328 RCIM_Cubic>,<at 0.5161 0.1208 RCIM_Cubic>]}
       }})
     Scale Factor=&MakeVector4FromVectorAndFloat("Vector (RGB)"=&MakeVectorFromLinearColorRGB(Color=$System.TOD), "Float (Alpha)"=1)
    )) {isSetParameters=true}
   /SetVariables(Particles.SpriteSize2=&Multiply_Vector2(
     A=$Particles.InitialSize
     B=&Vector2DFromFloat(Value=&Multiply_Float(
       A=&Subtract_Float(
        A=1
        B=&FloatFromCurve(
         FloatCurve=DataInterface() {propertyValues {Curve {keys [
             <at 0 0 RCIM_Cubic>
             {arriveTangent=-0.0175, interpMode=RCIM_Cubic, leaveTangent=-0.0175, tangentMode=RCTM_User, time=0.1563, value=0.907}
             {arriveTangent=-0.0042, interpMode=RCIM_Cubic, leaveTangent=-0.0042, tangentMode=RCTM_User, time=0.2646, value=0.0384}
             {arriveTangent=0.0081, interpMode=RCIM_Cubic, leaveTangent=0.0081, tangentMode=RCTM_User, time=0.4692, value=0.7123}
             {arriveTangent=0.034, interpMode=RCIM_Cubic, leaveTangent=0.034, tangentMode=RCTM_User, time=0.6406, value=0.0348}
             {arriveTangent=0.019, interpMode=RCIM_Cubic, leaveTangent=0.019, tangentMode=RCTM_User, time=0.7993, value=0.5053}
             <at 1 0 RCIM_Cubic>
            ]}}}
         Scale Curve=0.2
        )
       )
       B=&FloatFromCurve(
        FloatCurve=DataInterface() {propertyValues {Curve {keys [
            <at 0 0 RCIM_Cubic>
            {arriveTangent=-0.0042, interpMode=RCIM_Cubic, leaveTangent=-0.0042, tangentMode=RCTM_User, time=0.5, value=1}
            <at 1 0 RCIM_Cubic>
           ]}}}
        Scale Curve=0.5
       )
      ))
    )) {isSetParameters=true}
   &ScaleColor(
    Scale RGB=&Multiply_Vector(
     A=&MakeVectorFromLinearColorRGB(Color=$System.TOD)
     B=&VectorFromFloat(Value=&FloatFromCurve(FloatCurve=DataInterface() {propertyValues {Curve {keys [<at 0 0>,<at 0.1 1>,<at 1 0>]}}}))
    )
    ScaleA=true
    ScaleRGB=true
   )
   /SetVariables() {isSetParameters=true}
   /SetVariables(Particles.SnowSize:Vector2f=&Multiply_Vector2D_ByFloat(
     Float=&FloatFromCurve(
      FloatCurve=DataInterface() {propertyValues {Curve {keys [
          <at 0 0 RCIM_Cubic>
          {arriveTangent=1.6667, leaveTangent=1.6667, time=0.1, value=1}
          {arriveTangent=-1.1111, interpMode=RCIM_Cubic, leaveTangent=-1.1111, time=0.6, value=1}
          <at 1 0 RCIM_Cubic>
         ]}}}
      Scale Curve=&FN_FrontEndFloatModifier(Gameplay=1, Front End=$User.FrontEndScale)
     )
     Vector2D=&Vector2DFromFloat(Value=&RandomRangeFloat(Minimum=1.5, Maximum=4))
    )) {isSetParameters=true}
  ]
  rendererClasses [&NiagaraSpriteRendererProperties]
  renderers [
   {rendererClass=&NiagaraSpriteRendererProperties}
   {rendererClass=&NiagaraSpriteRendererProperties, rendererIndex=1}
   {rendererClass=&NiagaraSpriteRendererProperties, rendererIndex=2}
  ]
 }
 Trail {
  emitterUpdateScript [
   &EmitterState()
   &SpawnRate(
    SpawnGroup=1
    SpawnRate=&FloatFromCurve(
     FloatCurve=DataInterface() {propertyValues {Curve {keys [<at 0 1 RCIM_Constant>,<at 0.22 0 RCIM_Constant>]}}}
     CurveIndex=$Emitter.NormalizedLoopAge
     Scale Curve=100
    )
    Use Spawn Probability=false
   )
  ]
  particleSpawnScript [
   &PartitionParticles(Write Partition to Ribbon ID=::NewEnumerator0)
   &InitializeParticle(
    Color=&MultiplyLinearColor(Linear Color=<rgba 0.8125 0.8594 1 1>, Scale Factor=&MakeVector4_FromLinearColor(LinearColor=$System.TOD))
    Lifetime=0.7
    Ribbon Width=&Multiply_Float(A=5, B=&FN_FrontEndFloatModifier(Gameplay=1, Front End=$User.FrontEndScale))
    Ribbon Width Mode=::NewEnumerator1
    Sprite Rotation Mode=::NewEnumerator0
    UsePositionOffset=false
    Write Material Random=true
   )
   /SetVariables(Particles.Position=&Add_Vector(
     A=<0 0 6>
     B=&Add_Vector(
      A=&Lerp_Vector(
       A=$System.HeartIntroSpawnPosition
       B=$System.BurstPosition
       Alpha=&FloatFromCurve(
        FloatCurve=DataInterface() {propertyValues {Curve {keys [{arriveTangent=2.0116, interpMode=RCIM_Cubic, leaveTangent=2.0116, tangentMode=RCTM_User, time=0, value=0}, <at 1 1 RCIM_Cubic>]}}}
        CurveIndex=&Divide_Float(A=$Emitter.Age, B=$System.IntroLength)
       )
      )
      B=&MakeVector(
       X=&Waveform(
        Global Input Position=&MergeHelper_FloatSwitch(0=0, 1=1, 2=?, 3=?, Index=$Particles.PartitionParticles.Partition, Num=::NewEnumerator0)
        Global Phase=&Multiply_Float(A=$Emitter.Age, B=1.2)
        "Waveform [1]"=::NewEnumerator0
        "[1] Frequency"=1.5
        "[1] Amplitude Min/Max"=<-1 1>
        "[1] Scale"=&FloatFromCurve(
         FloatCurve=DataInterface() {propertyValues {Curve {keys [
             {arriveTangent=5.9356, interpMode=RCIM_Cubic, leaveTangent=5.9356, tangentMode=RCTM_User, time=-0.2504, value=-0.4466}
             <at 0.2496 0.5534 RCIM_Cubic>
            ]}}}
         CurveIndex=$Emitter.Age
         Scale Curve=&Multiply_Float(A=10, B=&FN_FrontEndFloatModifier(Gameplay=1, Front End=$User.FrontEndScale))
        )
       )
       Y=&Waveform(
        Global Input Position=&MergeHelper_FloatSwitch(0=0, 1=1, 2=?, 3=?, Index=$Particles.PartitionParticles.Partition, Num=::NewEnumerator0)
        Global Phase=&Multiply_Float(A=$Emitter.Age, B=1.2)
        "Waveform [1]"=::NewEnumerator1
        "[1] Frequency"=1.5
        "[1] Amplitude Min/Max"=<-1 1>
        "[1] Scale"=&FloatFromCurve(
         FloatCurve=DataInterface() {propertyValues {Curve {keys [
             {arriveTangent=5.9356, interpMode=RCIM_Cubic, leaveTangent=5.9356, tangentMode=RCTM_User, time=-0.2504, value=-0.4466}
             <at 0.2496 0.5534 RCIM_Cubic>
            ]}}}
         CurveIndex=$Emitter.Age
         Scale Curve=&Multiply_Float(A=10, B=&FN_FrontEndFloatModifier(Gameplay=1, Front End=$User.FrontEndScale))
        )
       )
       Z=0
      )
     )
    )) {isSetParameters=true}
   /SetVariables(Particles.RibbonUVDistance:NiagaraFloat=&DistanceBetweenPositions(
     A=&Lerp_Vector(
      A=$System.HeartIntroSpawnPosition
      B=$System.BurstPosition
      Alpha=&FloatFromCurve(
       FloatCurve=DataInterface() {propertyValues {Curve {keys [
           {arriveTangent=0.7913, interpMode=RCIM_Cubic, leaveTangent=0.7913, tangentMode=RCTM_User, time=0, value=0}
           {arriveTangent=-0.1073, tangentMode=RCTM_User, time=1, value=1}
          ]}}}
       CurveIndex=&Divide_Float(A=$Emitter.Age, B=$System.IntroLength)
      )
     )
     B=<0 0 0>
    )) {isSetParameters=true}
   /SetVariables(Particles.SpawnAge:NiagaraFloat=&Divide_Float(A=$Emitter.Age, B=$System.IntroLength)) {isSetParameters=true}
  ]
  particleUpdateScript [
   &ParticleState()
   &ScaleColor(
    Scale Alpha=&Multiply_Float(
     A=&FloatFromCurve(FloatCurve=DataInterface() {propertyValues {Curve {
         defaultValue=0
         keys [
          {arriveTangent=-0.2441, time=0.5, value=1}
          {arriveTangent=-0.0405, interpMode=RCIM_Cubic, leaveTangent=-0.0405, tangentMode=RCTM_User, time=1, value=0}
         ]
        }}})
     B=1.2
    )
    Scale RGB=&MakeVectorFromLinearColorRGB(Color=&ColorFromCurve(ColorCurve=DataInterface() {propertyValues {
        AlphaCurve {keys [<at 0 1>]}
        BlueCurve {keys [<at 0 0.9688>,<at 0.2975 0.3669>,<at 0.9945 0.1492>]}
        GreenCurve {keys [<at 0 0.7035>,<at 0.2975 0.3025>,<at 0.9945 0.0736>]}
        RedCurve {keys [<at 0 0.2294>,<at 0.2975 0.1688>,<at 0.9945 0.0273>]}
       }}))
    ScaleA=true
    ScaleRGB=true
   )
   &DynamicMaterialParameters(
    Index 0 Param 1=-0.5
    Index 0 Param 3=1
    Index 0 Param 4=1
    Module.Index0Float0Write=true
    Module.Index0Float1Write=true
    Module.Index0Float2Write=true
    Module.Index0Float3Write=true
   ) {module=DynamicMaterialParameters001}
   &ScaleRibbonWidth(Ribbon Width Scale=&Multiply_Float(
     A=&FloatFromCurve(FloatCurve=DataInterface() {propertyValues {Curve {keys [<at 0 0>,<at 0.1 1>,<at 1 2>]}}})
     B=&FloatFromCurve(FloatCurve=DataInterface() {propertyValues {Curve {keys [<at 0.05 0>,<at 0.15 1>,<at 0.8 1>,<at 0.9 0>]}}}, CurveIndex=$Particles.SpawnAge)
    ))
   &Drag(Drag=2, Use Linear Drag=true, Use Rotational Drag=true)
   &SolveForcesAndVelocity(Clamp Velocity=false, Limit Acceleration=false)
  ]
  rendererClasses [&NiagaraRibbonRendererProperties]
  renderers [{rendererClass=&NiagaraRibbonRendererProperties}]
 }
}
summary {
 systemName=NS_ReactFX_CaninePack
 emitters [
  {emitterName=Heart, rendererClasses [&NiagaraSpriteRendererProperties]}
  {emitterName=HeartGlow, rendererClasses [&NiagaraSpriteRendererProperties]}
  {emitterName=Trail, rendererClasses [&NiagaraRibbonRendererProperties]}
  {emitterName=LWE_AirGlow, rendererClasses []}
  {emitterName=LWE_Corona, rendererClasses []}
  {emitterName=Frosty_Smoke, rendererClasses [&NiagaraSpriteRendererProperties]}
  {emitterName=LWE_Rays001, rendererClasses []}
  {emitterName=Sparkles_And_Snow, rendererClasses [&NiagaraSpriteRendererProperties]}
  {emitterName=LWE_Rays, rendererClasses []}
  {emitterName=Pop_Shatter, rendererClasses [&NiagaraSpriteRendererProperties]}
  {emitterName=LWE_Mist, rendererClasses []}
 ]
 userVariables [
  User.bIsFrontEnd {type=NiagaraBool}
  User.bIsFrontEndPreview {type=NiagaraBool}
  User.FrontEndScale {type=NiagaraFloat, defaultValue=0.6, description=""}
 ]
}
