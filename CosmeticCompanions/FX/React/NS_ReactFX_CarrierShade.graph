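# Niagara reaction effect for cosmetic companions: a spinning, growing heart with a glow and a petting-reaction sound.
# Sparkles stream during the intro, then a streak burst and camera-aligned shatter sprites pop near its end.
# System-level IntroLength, OffsetPosition, BurstPosition and TOD values drive shared timing, placement and tint.
# Front-end modifiers and User.FrontEndScale adjust size, offsets and heart spin between gameplay and menus.
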NiagaraSystem /CosmeticCompanions/FX/React/NS_ReactFX_CarrierShade
dependencies {
 usedDataInterfaces [
  &NiagaraDataInterfaceAudioPlayer
  &NiagaraDataInterfaceCurve
  &NiagaraDataInterfaceColorCurve
  &NiagaraDataInterfaceSpriteRendererInfo
  &NiagaraDataInterfaceCamera
  &NiagaraDataInterfaceVector2DCurve
 ]
 usedDynamicInputs [
  &MakeCustomLinearColorFromBool
  &Boolean_OrOperation
  &MakeFloatFromVector
  &Divide_Float
 ]
 usedModules [
  &self:SystemSpawnScript.SystemScriptSource.SystemScriptGraph.NiagaraNodeAssignment_0.SetVariables_#1
  &self:SystemSpawnScript.SystemScriptSource.SystemScriptGraph.NiagaraNodeAssignment_6.SetVariables_#2
  &self:SystemSpawnScript.SystemScriptSource.SystemScriptGraph.NiagaraNodeAssignment_7.SetVariables_#3
  &self:SystemSpawnScript.SystemScriptSource.SystemScriptGraph.NiagaraNodeAssignment_5.SetVariables_#4
  &self:SystemSpawnScript.SystemScriptSource.SystemScriptGraph.NiagaraNodeAssignment_3.SetVariables_#5
  &self:SystemSpawnScript.SystemScriptSource.SystemScriptGraph.NiagaraNodeAssignment_1.SetVariables_#6
  &SystemState
  &self:SystemSpawnScript.SystemScriptSource.SystemScriptGraph.NiagaraNodeAssignment_8.SetVariables_#7
 ]
}
emitters {
 Heart {
  emitterSpawnScript [
   &InitializeAudioPlayer(Audio Player=DataInterface() {propertyValues {
      Attenuation=null
      Concurrency=null
      MaxPlaysPerTick=10
      SoundToPlay=/CosmeticCompanions/Assets/Audio_Shared/Interaction/Petting/SC_Mimosa_Interaction_React_Default
      bAllowLoopingOneShotSounds=false
      bLimitPlaysPerTick=true
      bOnlyActiveDuringGameplay=false
      bStopWhenComponentIsDestroyed=true
      ConfigurationUserParameter {parameter {name=None, typeDefHandle {registeredTypeIndex=1}, varData []}}
      ParameterNames []
     }})
   /Niagara/Modules/Audio/PlayAudio() {module=PlayAudio001}
  ]
  emitterUpdateScript [
   &EmitterState()
   &SpawnBurst_Instantaneous(Use Spawn Probability=false, UseLoopCountLimit=false)
  ]
  particleSpawnScript [
   /SetVariables(Particles.VisibilityTag=$Emitter.SpawnGroup) {isSetParameters=true}
   &InitializeParticle(
    Color=&MultiplyLinearColor(
     Linear Color=&MergeHelper_ColorSwitch(
      0=<rgba 1 1 1 1>
      1=<rgba 1 1 1 1>
      2=?
      3=?
      Index=&MakeIntFromFloat(Float=&FN_FrontEndFloatModifier(Gameplay=0, Front End=1))
      Num=::NewEnumerator0
     )
     Scale Factor=&MakeVector4_FromLinearColor(LinearColor=<rgba 1.5 1.5 1.5 1>)
    )
    Lifetime=$System.IntroLength
    Mesh Scale Mode=::NewEnumerator3
    Mesh Uniform Scale=0.7
    Sprite Rotation Mode=::NewEnumerator0
    Sprite Size Mode=::NewEnumerator3
    Uniform Sprite Size=&Multiply_Float(A=25, B=&FN_FrontEndFloatModifier(Gameplay=1, Front End=$User.FrontEndScale))
    UsePositionOffset=false
    Write Material Random=true
   )
   &InitialMeshOrientation(Mesh Axis To Orient=<0 1 0>, Mesh Orientation Mode=::NewEnumerator2, Orientation Vector=<0 0 0>, Use Rotation Vector=true)
   /SetVariables(Particles.GlowCamera=0.1) {isSetParameters=true}
   /SetVariables(Particles.Offset=-1) {isSetParameters=true}
   /SetVariables(Particles.VisibilityTag=&MakeIntFromFloat(Float=&FN_FrontEndFloatModifier(Gameplay=0, Front End=1))) {isSetParameters=true}
  ]
  particleUpdateScript [
   &ParticleState()
   &UpdateMeshOrientation(
    Rotation Rate=&FN_FrontEndFloatModifier(
     Gameplay=&FloatFromCurve(FloatCurve=DataInterface() {propertyValues {Curve {keys [<at 0 2.15>,<at 0.5 0>]}}}, Scale Curve=4)
     Front End=&FloatFromCurve(FloatCurve=DataInterface() {propertyValues {Curve {keys [<at 0 2.12>,<at 0.5 0>]}}}, Scale Curve=4)
    )
    Rotation Vector=<0 0 -1>
   )
   &ScaleMeshSize(Scale Factor=&VectorFromFloat(Value=&Add_Float(
      A=0
      B=&FloatFromCurve(FloatCurve=DataInterface() {propertyValues {Curve {
          defaultValue=0
          keys [
           <at 0 0.3 RCIM_Cubic>
           {arriveTangent=1.25, leaveTangent=1.25, time=0.4, value=1}
           {arriveTangent=6.566, interpMode=RCIM_Cubic, leaveTangent=6.566, tangentMode=RCTM_User, time=0.92, value=1.1}
           {arriveTangent=35.933, interpMode=RCIM_Cubic, leaveTangent=35.933, tangentMode=RCTM_User, time=1, value=3}
          ]
         }}})
     )))
   &ScaleSpriteSize(
    Scale Sprite Size Mode=::NewEnumerator0
    Uniform Scale Factor=&Add_Float(
     A=0
     B=&FloatFromCurve(FloatCurve=DataInterface() {propertyValues {Curve {
         defaultValue=0
         keys [
          <at 0 0.3 RCIM_Cubic>
          {arriveTangent=1.25, leaveTangent=1.25, time=0.4, value=1}
          {arriveTangent=6.566, interpMode=RCIM_Cubic, leaveTangent=6.566, tangentMode=RCTM_User, time=0.9, value=1.1}
          {arriveTangent=35.933, interpMode=RCIM_Cubic, leaveTangent=35.933, tangentMode=RCTM_User, time=1, value=3}
         ]
        }}})
    )
   ) {module=ScaleSpriteSize002}
   &ScaleColor(
    Scale Alpha=&FloatFromCurve(FloatCurve=DataInterface() {propertyValues {Curve {
        defaultValue=0
        keys [
         {arriveTangent=9.0725, interpMode=RCIM_Cubic, leaveTangent=9.0725, tangentMode=RCTM_User, time=0, value=0}
         {arriveTangent=-0.2441, leaveTangent=-0.2441, time=0.1, value=0.9847}
         {arriveTangent=-1.146, leaveTangent=-1.146, time=0.98, value=1}
         {arriveTangent=-0.0405, interpMode=RCIM_Cubic, leaveTangent=-0.0405, tangentMode=RCTM_User, time=1, value=0}
        ]
       }}})
    ScaleA=true
    ScaleRGB=true
   )
   /SetVariables(Particles.Position=$System.OffsetPosition) {isSetParameters=true}
   &DynamicMaterialParameters(
    Index 0 Param 1=&FloatFromCurve(FloatCurve=DataInterface() {propertyValues {Curve {keys [{interpMode=RCIM_Cubic, tangentMode=RCTM_Break, time=0, value=0}, {arriveTangent=2.3933, tangentMode=RCTM_User, time=1, value=1.65}]}}})
    Index 0 Param 2=&FloatFromCurve(FloatCurve=DataInterface() {propertyValues {Curve {keys [{arriveTangent=0.6296, interpMode=RCIM_Cubic, leaveTangent=0.6296, tangentMode=RCTM_User, time=0.5, value=1}, <at 0.6 50>, <at 0.7 1>]}}})
    Index 0 Param 3=-1
    Module.Index0Float0Write=true
    Module.Index0Float1Write=false
    Module.Index0Float2Write=true
    Module.Index0Float3Write=true
   )
   /SetVariables(Particles.Linear Color=&ColorFromCurve(ColorCurve=DataInterface() {propertyValues {
       AlphaCurve {keys [<at 0 1 RCIM_Cubic>,<at 1 1 RCIM_Cubic>]}
       BlueCurve {keys [<at 0 0.7563 RCIM_Cubic>,<at 1 1.5 RCIM_Cubic>]}
       GreenCurve {keys [<at 0 0.4716 RCIM_Cubic>,<at 1 1.5 RCIM_Cubic>]}
       RedCurve {keys [<at 0 1 RCIM_Cubic>,<at 1 1.5 RCIM_Cubic>]}
      }})) {isSetParameters=true}
   /SetVariables() {isSetParameters=true}
   /SetVariables(Particles.OutlineScale:Vector3f=&Multiply_Vector(A=$Particles.Scale, B=&VectorFromFloat(Value=1.1))) {isSetParameters=true}
  ]
  rendererClasses [&NiagaraMeshRendererProperties]
  renderers [
   {rendererClass=&NiagaraMeshRendererProperties, rendererIndex=0}
   {rendererClass=&NiagaraMeshRendererProperties, rendererIndex=1}
   {rendererClass=&NiagaraMeshRendererProperties, rendererIndex=2}
   {rendererClass=&NiagaraMeshRendererProperties, rendererIndex=3}
   {rendererClass=&NiagaraMeshRendererProperties, rendererIndex=4}
   {rendererClass=&NiagaraMeshRendererProperties, rendererIndex=5}
  ]
 }
 HeartGlow {
  emitterUpdateScript [
   &EmitterState()
   &SpawnBurst_Instantaneous(Use Spawn Probability=false, UseLoopCountLimit=false)
  ]
  particleSpawnScript [&InitializeParticle(
    Color=$System.TOD
    Lifetime=1.05
    Position Offset=$System.BurstPosition
    Sprite Rotation Angle=15
    Sprite Rotation Mode=::NewEnumerator0
    Sprite Size Mode=::NewEnumerator3
    Uniform Sprite Size=&Multiply_Float(A=25, B=&FN_FrontEndFloatModifier(Gameplay=1, Front End=$User.FrontEndScale))
    UsePositionOffset=true
    Write Material Random=true
   )]
  particleUpdateScript [
   &ParticleState()
   &ScaleSpriteSize(
    Scale Sprite Size Mode=::NewEnumerator0
    Uniform Scale Factor=&FloatFromCurve(
     FloatCurve=DataInterface() {propertyValues {Curve {keys [
         {arriveTangent=3.3458, interpMode=RCIM_Cubic, leaveTangent=3.3458, tangentMode=RCTM_User, time=0, value=1}
         {arriveTangent=1.251, tangentMode=RCTM_User, time=0.7, value=3}
         <at 0.9 2>
         <at 1.0056 5>
        ]}}}
     Scale Curve=&Waveform("Waveform [1]"=::NewEnumerator12, "[1] Frequency"=5, "[1] Amplitude Min/Max"=<0.7 1>)
    )
   ) {module=ScaleSpriteSize002}
   &ScaleColor(
    Scale Alpha=&FloatFromCurve(FloatCurve=DataInterface() {propertyValues {Curve {
        defaultValue=0
        keys [
         {arriveTangent=0.6582, interpMode=RCIM_Cubic, leaveTangent=0.6582, tangentMode=RCTM_User, time=0.3, value=0}
         {arriveTangent=1.6667, interpMode=RCIM_Cubic, leaveTangent=1.6667, time=0.7134, value=0.3}
         {arriveTangent=0.3137, interpMode=RCIM_Cubic, leaveTangent=0.3137, tangentMode=RCTM_User, time=0.9, value=1}
        ]
       }}})
    Scale RGB=&MakeVectorFromLinearColorRGB(Color=&ColorFromCurve(ColorCurve=DataInterface() {propertyValues {
        AlphaCurve {keys [<at 0 1 RCIM_Cubic>]}
        BlueCurve {keys [<at 0.3291 0.664 RCIM_Cubic>,<at 0.8239 0.9868>,<at 1 1 RCIM_Cubic>]}
        GreenCurve {keys [<at 0.3291 0.2113 RCIM_Cubic>,<at 0.8239 0.3568>,<at 1 0.2849 RCIM_Cubic>]}
        RedCurve {keys [<at 0.3291 1 RCIM_Cubic>,<at 0.8239 1>,<at 1 0.8863 RCIM_Cubic>]}
       }}))
    ScaleA=true
    ScaleRGB=true
   )
   /SetVariables(Particles.Position=$System.OffsetPosition) {isSetParameters=true}
   &DynamicMaterialParameters(
    Index 0 Param 1=1
    Index 0 Param 3=1
    Module.Index0Float0Write=true
    Module.Index0Float1Write=true
    Module.Index0Float2Write=true
    Module.Index0Float3Write=true
   )
   &FN_Sprite_Attributes(
    Color=&MakeLinearColorFromVectorAndFloat(
     "Vector (RGB)"=&MakeVectorFromLinearColorRGB(Color=<rgba 1 0.2503 0.8261 1>)
     "Float (Alpha)"=&FloatFromCurve(
      FloatCurve=DataInterface() {propertyValues {Curve {
         defaultValue=0
         keys [
          {arriveTangent=0.6582, interpMode=RCIM_Cubic, leaveTangent=0.6582, tangentMode=RCTM_User, time=0, value=0}
          {arriveTangent=0.3137, interpMode=RCIM_Cubic, leaveTangent=0.3137, tangentMode=RCTM_User, time=0.8, value=1}
          <at 0.9832 0 RCIM_Cubic>
         ]
        }}}
      Scale Curve=0.2
     )
    )
    Create Color Attribute=true
    Create SpriteSize Attribute=true
    SpriteSize=&Vector2DFromFloat(Value=&Multiply_Float(
      A=&Multiply_Float(A=20, B=&FN_FrontEndFloatModifier(Gameplay=1, Front End=$User.FrontEndScale))
      B=&FloatFromCurve(FloatCurve=DataInterface() {propertyValues {Curve {keys [
           {arriveTangent=3.3458, interpMode=RCIM_Cubic, leaveTangent=3.3458, tangentMode=RCTM_User, time=0, value=1}
           {arriveTangent=1.251, tangentMode=RCTM_User, time=0.5, value=3}
           <at 0.8 3>
           <at 0.9809 5>
          ]}}})
     ))
    Write to Module Namespace=true
   )
  ]
  rendererClasses [&NiagaraSpriteRendererProperties]
  renderers [
   {rendererClass=&NiagaraSpriteRendererProperties}
   {rendererClass=&NiagaraSpriteRendererProperties, rendererIndex=1}
  ]
 }
 Intro_Sparkles {
  emitterUpdateScript [
   &EmitterState(
    Life Cycle Mode=::NewEnumerator1
    Loop Behavior=::NewEnumerator1
    Loop Delay=0.47
    Loop Duration=&Add_Float(A=$System.IntroLength, B=-0.2)
    UseLoopDelay=false
   )
   &SpawnRate(
    SpawnGroup=1
    SpawnRate=&FloatFromCurve(
     FloatCurve=DataInterface() {propertyValues {Curve {keys [<at 0 1>,<at 0.77 1>,<at 0.8 0>]}}}
     CurveIndex=$Emitter.NormalizedLoopAge
     Scale Curve=20
    )
    Use Spawn Probability=false
   ) {module=SpawnRate001}
   &SpawnRate(SpawnRate=20, Use Spawn Probability=false)
  ]
  particleSpawnScript [
   &InitializeParticle(
    Color Maximum=<rgba 1 0.222 0.7039 1>
    Color Minimum=<rgba 1 0.4045 0.8917 1>
    Color Mode=::NewEnumerator2
    Lifetime Max=0.15
    Lifetime Min=0.1
    Lifetime Mode=::NewEnumerator1
    Sprite Size Max=&MergeHelper_Vector2Switch(0=<4 10>, 1=<2 30>, 2=?, 3=?, 4=?, Index=&SpawnGroup(), Num=::NewEnumerator0)
    Sprite Size Min=&MergeHelper_Vector2Switch(0=<5 10>, 1=<2 20>, 2=?, 3=?, 4=?, Index=&SpawnGroup(), Num=::NewEnumerator0)
    Sprite Size Mode=::NewEnumerator2
    UsePositionOffset=true
    Write Material Random=true
   )
   &CameraOffset(Camera Offset Amount=10)
   /SetVariables(Particles.VisibilityTag=&SpawnGroup()) {isSetParameters=true, module=SetVariables_#8}
   &ApplyInitialForces()
   &ShapeLocation(
    MaskSpawnGroup=false
    Offset=&FN_FrontEndVector3Modifier(
     Gameplay=&Lerp_Vector(
      A=<0 0 -50>
      B=<0 0 0>
      Alpha=&FloatFromCurve(
       FloatCurve=DataInterface() {propertyValues {Curve {defaultValue=1, keys [{tangentMode=RCTM_User, time=0, value=0}, {tangentMode=RCTM_User, time=0.6, value=1}]}}}
       CurveIndex=$Emitter.Age
      )
     )
     Front End=&Lerp_Vector(
      A=<0 0 -50>
      B=<0 0 -22>
      Alpha=&FloatFromCurve(
       FloatCurve=DataInterface() {propertyValues {Curve {defaultValue=1, keys [{tangentMode=RCTM_User, time=0, value=0}, {tangentMode=RCTM_User, time=0.6, value=1}]}}}
       CurveIndex=$Emitter.Age
      )
     )
    )
    Offset Mode=::NewEnumerator0
    Override Seed=false
    Sphere Radius=&MergeHelper_FloatSwitch(0=13, 1=10, 2=?, 3=?, Index=&SpawnGroup(), Num=::NewEnumerator0)
    UseOwnerScale=true
   ) {module=ShapeLocation001}
   &AddVelocity(Velocity=&MergeHelper_VectorSwitch(
     0=<0 0 50>
     1=&Multiply_VectorByFloat(
      Vector=<0 0 -150>
      Float=&FloatFromCurve(FloatCurve=DataInterface() {propertyValues {Curve {keys [<at 0 1>,<at 0.5 1>,<at 1 0>]}}}, CurveIndex=$Emitter.NormalizedLoopAge)
     )
     2=?
     3=?
     Index=&SpawnGroup()
     Num=::NewEnumerator0
    ))
  ]
  particleUpdateScript [
   &ParticleState()
   &ShapeLocation(
    MaskSpawnGroup=false
    Non Uniform Scale=<1 1 0.8>
    Offset=&Lerp_Vector(
     A=$System.HeartIntroSpawnPosition
     B=$System.BurstPosition
     Alpha=&FloatFromCurve(
      FloatCurve=DataInterface() {propertyValues {Curve {keys [
          {arriveTangent=0.7913, interpMode=RCIM_Cubic, leaveTangent=0.7913, tangentMode=RCTM_User, time=0, value=0}
          {arriveTangent=0.2071, interpMode=RCIM_Cubic, leaveTangent=0.2071, tangentMode=RCTM_User, time=0.7, value=0.9784}
          <at 1 1 RCIM_Cubic>
         ]}}}
      CurveIndex=$Emitter.NormalizedLoopAge
     )
    )
    Offset Mode=::NewEnumerator0
    Override Seed=false
    Sphere Radius=10
    UseOwnerScale=true
   ) {enabled=false}
   &ScaleSpriteSize(
    Non-Uniform Curve Sprite Scale=DataInterface() {propertyValues {
      XCurve {keys [
        <at 0 0 RCIM_Cubic>
        {arriveTangentWeight=0.0872, interpMode=RCIM_Cubic, leaveTangentWeight=0.3197, tangentMode=RCTM_User, time=0.1569, value=1.625}
        {arriveTangentWeight=0.0926, interpMode=RCIM_Cubic, leaveTangentWeight=0.2166, tangentMode=RCTM_User, time=0.3622, value=1}
        {arriveTangentWeight=0.2824, interpMode=RCIM_Cubic, leaveTangentWeight=0.1558, tangentMode=RCTM_User, time=0.9005, value=1}
        <at 1.0034 0 RCIM_Cubic>
       ]}
      YCurve {keys [
        <at 0 0 RCIM_Cubic>
        {arriveTangentWeight=0.0872, interpMode=RCIM_Cubic, leaveTangentWeight=0.3197, tangentMode=RCTM_User, time=0.1534, value=1.625}
        {arriveTangentWeight=0.0676, interpMode=RCIM_Cubic, leaveTangentWeight=0.163, tangentMode=RCTM_User, time=0.3553, value=1}
        {arriveTangentWeight=0.2837, interpMode=RCIM_Cubic, leaveTangentWeight=0.153, tangentMode=RCTM_User, time=0.9005, value=1}
        <at 0.9966 2 RCIM_Cubic>
       ]}
     }}
    Scale Sprite Size Mode=::NewEnumerator3
   )
   &ScaleColor(
    Scale Alpha=&FloatFromCurve(
     FloatCurve=DataInterface() {propertyValues {Curve {keys [
         <at 0 0 RCIM_Cubic>
         {arriveTangent=0.4409, interpMode=RCIM_Cubic, leaveTangent=0.4409, time=0.1023, value=1}
         {arriveTangent=-1.114, interpMode=RCIM_Cubic, leaveTangent=-1.114, time=0.6137, value=0.2706}
         <at 1 0 RCIM_Cubic>
        ]}}}
     Scale Curve=&RandomRangeFloat(Minimum=0.5, Maximum=1)
    )
    ScaleA=true
    ScaleRGB=true
   )
   &SolveForcesAndVelocity(Clamp Velocity=false, Limit Acceleration=false)
  ]
  rendererClasses [&NiagaraSpriteRendererProperties]
  renderers [
   {rendererClass=&NiagaraSpriteRendererProperties}
   {rendererClass=&NiagaraSpriteRendererProperties, rendererIndex=1}
  ]
 }
 Intro_Sparkles001 {
  emitterUpdateScript [
   &EmitterState(Life Cycle Mode=::NewEnumerator1, Loop Behavior=::NewEnumerator1, Loop Delay=$System.IntroLength, Loop Duration=1, UseLoopDelay=true)
   &SpawnBurst_Instantaneous(Spawn Count=10, Use Spawn Probability=false, UseLoopCountLimit=false)
  ]
  particleSpawnScript [
   &InitializeParticle(
    Color Maximum=<rgba 1 0.222 0.7039 1>
    Color Minimum=<rgba 1 0.4045 0.8917 1>
    Color Mode=::NewEnumerator2
    Lifetime Max=0.12
    Lifetime Min=0.08
    Lifetime Mode=::NewEnumerator1
    Position Offset=&FN_FrontEndVector3Modifier(Gameplay=<0 0 0>, Front End=<0 0 -20>)
    Sprite Size Max=<5 80>
    Sprite Size Min=<5 50>
    Sprite Size Mode=::NewEnumerator2
    UsePositionOffset=true
    Write Material Random=true
   )
   &CameraOffset(Camera Offset Amount=10)
   /SetVariables(Particles.VisibilityTag=&SpawnGroup()) {isSetParameters=true, module=SetVariables_#8}
   &ShapeLocation(MaskSpawnGroup=false, Offset Mode=::NewEnumerator0, Override Seed=false, Sphere Radius=13, UseOwnerScale=true) {module=ShapeLocation001}
   &AddVelocity(
    Module.Use Falloff Distance=false
    Origin Offset=&FN_FrontEndVector3Modifier(Gameplay=<0 0 0>, Front End=<0 0 -20>)
    Origin Offset Coordinate Space=::Simulation
    Override Seed=false
    Velocity Mode=::NewEnumerator1
    Velocity Speed=&RandomRangeFloat(Minimum=100, Maximum=300)
   )
   &ApplyInitialForces()
  ]
  particleUpdateScript [
   &ParticleState()
   &ShapeLocation(
    MaskSpawnGroup=false
    Non Uniform Scale=<1 1 0.8>
    Offset=&Lerp_Vector(
     A=$System.HeartIntroSpawnPosition
     B=$System.BurstPosition
     Alpha=&FloatFromCurve(
      FloatCurve=DataInterface() {propertyValues {Curve {keys [
          {arriveTangent=0.7913, interpMode=RCIM_Cubic, leaveTangent=0.7913, tangentMode=RCTM_User, time=0, value=0}
          {arriveTangent=0.2071, interpMode=RCIM_Cubic, leaveTangent=0.2071, tangentMode=RCTM_User, time=0.7, value=0.9784}
          <at 1 1 RCIM_Cubic>
         ]}}}
      CurveIndex=$Emitter.NormalizedLoopAge
     )
    )
    Offset Mode=::NewEnumerator0
    Override Seed=false
    Sphere Radius=10
    UseOwnerScale=true
   ) {enabled=false}
   &ScaleSpriteSize(
    Non-Uniform Curve Sprite Scale=DataInterface() {propertyValues {
      XCurve {keys [
        <at 0 0 RCIM_Cubic>
        {arriveTangentWeight=0.0872, interpMode=RCIM_Cubic, leaveTangentWeight=0.3197, tangentMode=RCTM_User, time=0.1569, value=1.625}
        {arriveTangentWeight=0.0926, interpMode=RCIM_Cubic, leaveTangentWeight=0.2166, tangentMode=RCTM_User, time=0.3622, value=1}
        {arriveTangentWeight=0.2824, interpMode=RCIM_Cubic, leaveTangentWeight=0.1558, tangentMode=RCTM_User, time=0.9005, value=1}
        <at 1.0034 0 RCIM_Cubic>
       ]}
      YCurve {keys [
        <at 0 0 RCIM_Cubic>
        {arriveTangentWeight=0.0872, interpMode=RCIM_Cubic, leaveTangentWeight=0.3197, tangentMode=RCTM_User, time=0.1534, value=1.625}
        {arriveTangentWeight=0.0676, interpMode=RCIM_Cubic, leaveTangentWeight=0.163, tangentMode=RCTM_User, time=0.3553, value=1}
        {arriveTangentWeight=0.2837, interpMode=RCIM_Cubic, leaveTangentWeight=0.153, tangentMode=RCTM_User, time=0.9005, value=1}
        <at 0.9966 2 RCIM_Cubic>
       ]}
     }}
    Scale Sprite Size Mode=::NewEnumerator3
   )
   &ScaleColor(
    Scale Alpha=&FloatFromCurve(
     FloatCurve=DataInterface() {propertyValues {Curve {keys [
         <at 0 0 RCIM_Cubic>
         {arriveTangent=0.4409, interpMode=RCIM_Cubic, leaveTangent=0.4409, time=0.1023, value=1}
         {arriveTangent=-1.114, interpMode=RCIM_Cubic, leaveTangent=-1.114, time=0.6137, value=0.2706}
         <at 1 0 RCIM_Cubic>
        ]}}}
     Scale Curve=&RandomRangeFloat(Minimum=0.5, Maximum=1)
    )
    ScaleA=true
    ScaleRGB=true
   )
   &SolveForcesAndVelocity(Clamp Velocity=false, Limit Acceleration=false)
  ]
  rendererClasses [&NiagaraSpriteRendererProperties]
  renderers [{rendererClass=&NiagaraSpriteRendererProperties}]
 }
 LWE_AirGlow {rendererClasses [], renderers []}
 LWE_Rays {rendererClasses [], renderers []}
 LWE_Rays001 {rendererClasses [], renderers []}
 Pop_Shatter {
  emitterUpdateScript [
   &EmitterState()
   &SpawnBurst_Instantaneous(
    Spawn Count=35
    Spawn Group=1
    Spawn Time=&Add_Float(A=$System.IntroLength, B=-0.1)
    Use Spawn Probability=false
    UseLoopCountLimit=false
   )
  ]
  particleSpawnScript [
   &PartitionParticles(Number Of Partitions=9)
   &InitializeParticle(
    Color=&MultiplyLinearColor(Linear Color=$System.TOD, Scale Factor=&MakeVector4_FromLinearColor(LinearColor=<rgba 1 1 1 1>))
    Lifetime=&RandomRangeFloat(Minimum=0.2, Maximum=0.35)
    Position Offset=&FN_FrontEndVector3Modifier(Gameplay=<0 0 0>, Front End=<0 0 -20>)
    Sprite Rotation Angle=180
    Sprite Rotation Mode=::NewEnumerator0
    Sprite Size Mode=::NewEnumerator3
    Uniform Sprite Size=&Multiply_Float(
     A=&RandomRangeFloat(Minimum=5, Maximum=15)
     B=&FN_FrontEndFloatModifier(Gameplay=1, Front End=$User.FrontEndScale)
    )
    UsePositionOffset=true
    Write Material Random=true
   ) {module=InitializeParticle001}
   &SubUVAnimation(
    End Frame Range Override=9
    "Sprite Renderer "=DataInterface() {propertyValues {SpriteRenderer=&self:Pop_Shatter_0.NiagaraSpriteRendererProperties_0}}
    Start Frame Offset:NiagaraInt32=$Particles.PartitionParticles.Partition
    UseEndFrame=true
    UseStartFrame=false
   )
   /SetVariables(Particles.Position=&AddVectorToPosition(
     Position=&AddVectorToPosition(Position=Linked() {linkedVariable {name=System.BurstPosition, type=Vector3f}}, Vector=<0 0 0>)
     Vector=&Multiply_VectorByFloat(
      Vector=&RotateVectorByQuaternion(
       Vector To Rotate=<0 0 1>
       Quaternion=&MakeQuaternion(
        Quaternion From=0
        Axis=&NormalizeVector(
         Vector To Normalize=&Subtract_Position(
          A=&ReturnCameraPosition(
           Coordinate Space:ENiagaraCoordinateSpace=::Simulation
           Camera Query:NiagaraDataInterfaceCamera=DataInterface() {propertyValues {bRequireCurrentFrameData=false}}
          )
          B=/Niagara/DynamicInputs/Helpers/SimulationPosition()
         )
         Safe Normalize=true
         Safe Fallback Threshold=0
         Safe Fallback Vector=<0 0 0.001>
        )
        Angle=&Subtract_Float(A=&Multiply_FloatByInt(Float=40, Integer=$Particles.PartitionParticles.Partition), B=30)
        From=?
        To=?
        Angle Type=::NewEnumerator0
        Normalize Vectors=false
       )
      )
      Float=5
     )
    )) {enabled=false, isSetParameters=true}
   &AddVelocity(
    Velocity=&RotateVectorByQuaternion(
     Vector To Rotate=<0 0 1>
     Quaternion=&MakeQuaternion(
      Quaternion From=0
      Axis=&NormalizeVector(
       Vector To Normalize=&Subtract_Position(
        A=&ReturnCameraPosition(
         Coordinate Space:ENiagaraCoordinateSpace=::Simulation
         Camera Query:NiagaraDataInterfaceCamera=DataInterface() {propertyValues {bRequireCurrentFrameData=false}}
        )
        B=&SimulationPosition()
       )
       Safe Normalize=true
       Safe Fallback Threshold=0
       Safe Fallback Vector=<0 0 0.001>
      )
      Angle=&Subtract_Float(A=&Multiply_FloatByInt(Float=40, Integer=$Particles.PartitionParticles.Partition), B=30)
      From=?
      To=?
      Angle Type=::NewEnumerator0
      Normalize Vectors=false
     )
    )
    Velocity Speed Scale=&RandomRangeFloat(Minimum=50, Maximum=100)
   ) {enabled=false}
   &ShapeLocation(
    Cylinder Height=0
    Cylinder Radius=0
    MaskSpawnGroup=false
    Override Seed=false
    Rotation Axis=&ReturnCameraVector(Camera Vectors=::NewEnumerator4, Coordinate Space=::Simulation, Origin Position=$Particles.Position)
    Rotation Mode=::NewEnumerator1
    Shape Primitive=::NewEnumerator1
    Surface Only=false
    UseOwnerScale=true
   )
   &AddVelocity(
    Module.Use Falloff Distance=false
    Override Seed=false
    Velocity Mode=::NewEnumerator1
    Velocity Origin=&FN_FrontEndVector3Modifier(Gameplay=$System.BurstPosition, Front End=<0 0 -20>)
    Velocity Speed=&RandomRangeFloat(Minimum=300, Maximum=350)
   ) {module=AddVelocity001}
  ]
  particleUpdateScript [
   &ParticleState()
   &ScaleSpriteSize(
    Scale Factor=&MakeVector2D(
     X=1
     Y=&FloatFromCurve(FloatCurve=DataInterface() {propertyValues {Curve {keys [<at 0 1.5>,<at 1 1>]}}}, Scale Curve=1.1)
    )
    Scale Sprite Size Mode=::NewEnumerator2
   ) {module=ScaleSpriteSize001}
   &Drag(
    Drag=&Multiply_Float(
     A=&RandomRangeFloat(Minimum=20, Maximum=30)
     B=&FloatFromCurve(FloatCurve=DataInterface() {propertyValues {Curve {
         defaultValue=1
         keys [
          {interpMode=RCIM_Cubic, tangentMode=RCTM_User, time=0, value=0}
          {arriveTangent=1, interpMode=RCIM_Cubic, leaveTangent=1, time=0.5, value=1}
          {interpMode=RCIM_Cubic, tangentMode=RCTM_User, time=1, value=1}
         ]
        }}})
    )
    Use Linear Drag=true
    Use Rotational Drag=true
   )
   &SolveForcesAndVelocity(Clamp Velocity=false, Limit Acceleration=false)
   &ScaleColor(
    Scale Alpha=&FloatFromCurve(FloatCurve=DataInterface() {propertyValues {Curve {keys [{arriveTangent=0.317, interpMode=RCIM_Cubic, leaveTangent=0.317, tangentMode=RCTM_User, time=0.5, value=1}, <at 1 0 RCIM_Cubic>]}}})
    Scale RGB=&MakeVectorFromLinearColorRGB(Color=&ColorFromCurve(ColorCurve=DataInterface() {propertyValues {
        AlphaCurve {keys [<at 0 1 RCIM_Cubic>,<at 1 0 RCIM_Cubic>]}
        BlueCurve {keys [
          <at 0.2576 10>
          {arriveTangent=-12.5706, interpMode=RCIM_Cubic, leaveTangent=-12.5706, time=0.3833, value=5}
          <at 0.5307 1.0773>
          <at 0.9735 1 RCIM_Cubic>
         ]}
        GreenCurve {keys [
          <at 0.2576 10>
          {arriveTangent=-12.5706, interpMode=RCIM_Cubic, leaveTangent=-12.5706, time=0.3833, value=5}
          <at 0.5307 1.0773>
          <at 0.9735 1 RCIM_Cubic>
         ]}
        RedCurve {keys [
          <at 0.2576 10>
          {arriveTangent=-12.5706, interpMode=RCIM_Cubic, leaveTangent=-12.5706, time=0.3833, value=5}
          <at 0.5307 1.0773>
          <at 0.9735 1 RCIM_Cubic>
         ]}
       }}))
    ScaleA=true
    ScaleRGB=true
   )
   &ScaleSpriteSize(Scale Factor=&Vector2DFromFloat(Value=&FloatFromCurve(FloatCurve=DataInterface() {propertyValues {Curve {keys [<at 0 0.5 RCIM_Cubic>, {arriveTangent=2.1241, interpMode=RCIM_Cubic, leaveTangent=2.1241, tangentMode=RCTM_User, time=0.1, value=1}]}}})))
  ]
  rendererClasses [&NiagaraSpriteRendererProperties]
  renderers [{rendererClass=&NiagaraSpriteRendererProperties}]
 }
}
summary {
 systemName=NS_ReactFX_CarrierShade
 emitters [
  {emitterName=Heart, rendererClasses [&NiagaraMeshRendererProperties]}
  {emitterName=HeartGlow, rendererClasses [&NiagaraSpriteRendererProperties]}
  {emitterName=LWE_AirGlow, rendererClasses []}
  {emitterName=LWE_Rays001, rendererClasses []}
  {emitterName=LWE_Rays, rendererClasses []}
  {emitterName=Pop_Shatter, rendererClasses [&NiagaraSpriteRendererProperties]}
  {emitterName=Intro_Sparkles, rendererClasses [&NiagaraSpriteRendererProperties]}
  {emitterName=Intro_Sparkles001, rendererClasses [&NiagaraSpriteRendererProperties]}
 ]
 userVariables [
  User.bIsFrontEnd {type=NiagaraBool}
  User.bIsFrontEndPreview {type=NiagaraBool}
  User.FrontEndScale {type=NiagaraFloat, defaultValue=0.6, description=""}
 ]
}
