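# Niagara reaction effect for the Comic Tabby cosmetic companion, apparently a heart pop that ends in an ink burst.
# Intro plays the shared petting-reaction sound and grows layered sprites with a stepped flicker before collapse.
# After a one-second delay, InkSplat spawns a socket-sampled animated splat while Drops sprays short-lived droplets under gravity and drag.
# Shared system variables set burst position and intro length, and front-end modifiers shrink sizes for menu display.
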NiagaraSystem /CosmeticCompanions/FX/React/NS_ReactFX_ComicTabby
dependencies {
 usedDataInterfaces [
  &NiagaraDataInterfaceCurve
  &NiagaraDataInterfaceSpriteRendererInfo
  &NiagaraDataInterfaceAudioPlayer
  &NiagaraDataInterfaceVector2DCurve
 ]
 usedDynamicInputs [
  &MaskFloatBySpawnGroup
  &Lerp_Float
  &NormalizeVector
  &SubtractVectorFromPosition
  /Niagara/DynamicInputs/Add/Add_Vector2
  &Vector2DFromCurve
  &FN_FrontEndVector3Modifier
 ]
 usedModules [
  &self:SystemSpawnScript.SystemScriptSource.SystemScriptGraph.NiagaraNodeAssignment_0.SetVariables_#1
  &self:SystemSpawnScript.SystemScriptSource.SystemScriptGraph.NiagaraNodeAssignment_3.SetVariables_#2
  &self:SystemSpawnScript.SystemScriptSource.SystemScriptGraph.NiagaraNodeAssignment_8.SetVariables_#3
  &self:SystemSpawnScript.SystemScriptSource.SystemScriptGraph.NiagaraNodeAssignment_9.SetVariables_#4
  &self:SystemSpawnScript.SystemScriptSource.SystemScriptGraph.NiagaraNodeAssignment_10.SetVariables_#5
  &self:SystemSpawnScript.SystemScriptSource.SystemScriptGraph.NiagaraNodeAssignment_5.SetVariables_#6
  &self:SystemSpawnScript.SystemScriptSource.SystemScriptGraph.NiagaraNodeAssignment_6.SetVariables_#7
  &SystemState
 ]
}
emitters {
 Drops {
  emitterSpawnScript [/SetVariables(Emitter.RandomVec:Vector3f=&RandomVector(Vector Scale=1)) {isSetParameters=true}]
  emitterUpdateScript [
   &EmitterState(Life Cycle Mode=::NewEnumerator1, Loop Behavior=::NewEnumerator1, Loop Delay=1, Loop Duration=1, UseLoopDelay=true)
   &SpawnBurst_Instantaneous(Spawn Count=7, Use Spawn Probability=false, UseLoopCountLimit=false)
  ]
  particleSpawnScript [
   &InitializeParticle(
    Color=<rgba 0 0 0 1>
    Lifetime Max=0.3
    Lifetime Min=0.2
    Lifetime Mode=::NewEnumerator1
    Position Offset=$System.BurstPosition
    Sprite Size Mode=::NewEnumerator1
    Uniform Sprite Size Max=6
    Uniform Sprite Size Min=3
    UsePositionOffset=true
    Write Material Random=true
   )
   &ShapeLocation(
    MaskSpawnGroup=false
    Rotation Mode=::NewEnumerator2
    Sphere Distribution=::NewEnumerator3
    Sphere Radius=5
    UseOwnerScale=true
    Yaw / Pitch / Roll=$Emitter.RandomVec
   )
   &AddVelocity(
    Module.Use Falloff Distance=false
    Override Seed=false
    Velocity Mode=::NewEnumerator1
    Velocity Origin:Vector3f=$System.BurstPosition
    Velocity Speed=&RandomRangeFloat(Minimum=90, Maximum=300)
   )
   &CameraOffset(Camera Offset Amount=12)
  ]
  particleUpdateScript [
   &ParticleState()
   &GravityForce(Gravity=&Multiply_VectorByFloat(Vector=<0 0 -900>, Float=&FN_FrontEndFloatModifier(Gameplay=1, Front End=0.65)))
   &Drag(Drag=3, Rotational Drag=3, Use Linear Drag=true, Use Rotational Drag=true) {module=Drag001}
   &SolveForcesAndVelocity(Clamp Velocity=false, Limit Acceleration=false)
   &DynamicMaterialParameters(
    Index 0 Param 1=1
    Index 0 Param 2=10
    Index 0 Param 3=68.1455
    Index 0 Param 4=5
    Module.Param0WriteEnabled=true
    Module.Param1WriteEnabled=true
    Module.Param2WriteEnabled=true
    Module.Param3WriteEnabled=true
   )
   &ScaleSpriteSize(
    Uniform Curve Scale=&FN_FrontEndFloatModifier(Gameplay=1, Front End=0.65)
    Uniform Curve Sprite Scale=DataInterface() {propertyValues {Curve {keys [
        {arriveTangent=40.4327, interpMode=RCIM_Cubic, leaveTangent=40.4327, tangentMode=RCTM_User, time=0, value=0.7}
        {arriveTangent=-9.8929, interpMode=RCIM_Cubic, leaveTangent=-9.8929, tangentMode=RCTM_User, time=0.07, value=1.5}
        <at 0.16 0.8>
        {interpMode=RCIM_Cubic, tangentMode=RCTM_User, time=0.3741, value=0.8}
        {arriveTangent=-1.9212, interpMode=RCIM_Cubic, leaveTangent=-1.9212, tangentMode=RCTM_User, time=1, value=0}
       ]}}}
   )
   /SetVariables(
    Particles.HeartWhite_DMP:Vector4f=&MakeVector4(X=0.9088, Y=0.026, Z=0, W=0)
    Particles.White:LinearColor=<rgba 2 2 2 1>
    Particles.WhiteScale:Vector2f=&Multiply_Vector2D_ByFloat(Float=&FN_FrontEndFloatModifier(Gameplay=2, Front End=1.3), Vector2D=$Particles.SpriteSize)
   ) {isSetParameters=true}
  ]
  rendererClasses [&NiagaraSpriteRendererProperties]
  renderers [
   {rendererClass=&NiagaraSpriteRendererProperties}
   {rendererClass=&NiagaraSpriteRendererProperties, rendererIndex=1}
  ]
 }
 InkSplat {
  emitterUpdateScript [
   &EmitterState(Life Cycle Mode=::NewEnumerator1, Loop Behavior=::NewEnumerator1, Loop Delay=1, Loop Duration=1, UseLoopDelay=true)
   &SpawnBurst_Instantaneous(Use Spawn Probability=false, UseLoopCountLimit=false)
  ]
  particleSpawnScript [
   &SkeletalMeshLocation(
    MaskSpawnGroup=false
    Mesh Sampling Type=::NewEnumerator5
    Orientation Sampling=::NewEnumerator1
    Position Sampling=::NewEnumerator2
    Sampled Position Offset=<-1 0 0.5>
    Sampled Position Offset Coordinate Space:ENiagaraSkelMeshTransforms=::NewEnumerator2
    Skeletal Mesh=$System.Skeletal Mesh
    Socket Sampling Mode=::NewEnumerator5
   )
   &InitializeParticle(
    Color=<rgba 0.2248 1 0.6955 1>
    Lifetime=0.2
    Position Offset=$System.BurstPosition
    Sprite Rotation Mode=::NewEnumerator2
    Sprite Size Mode=::NewEnumerator3
    Sprite UV Mode=::NewEnumerator3
    Uniform Sprite Size=&Multiply_Float(
     A=&RandomRangeFloat(Minimum=130, Maximum=145)
     B=&FN_FrontEndFloatModifier(Gameplay=1, Front End=0.65)
    )
    UsePositionOffset=true
    Write Material Random=true
   )
   &SubUVAnimation(
    "Sprite Renderer "=DataInterface() {propertyValues {SpriteRenderer=&self:InkSplat_0.NiagaraSpriteRendererProperties_1}}
    SubUV Animation Mode:ENiagara_SubUVLookupModeV2=::NewEnumerator2
    UseEndFrame=false
    UseStartFrame=false
   )
   &CameraOffset(Camera Offset Amount=2)
  ]
  particleUpdateScript [
   &SkeletalMeshLocation(
    Bone / Socket Index=1
    MaskSpawnGroup=false
    Mesh Sampling Type=::NewEnumerator5
    Orientation Sampling=::NewEnumerator1
    Position Sampling=::NewEnumerator2
    Skeletal Mesh=$System.Skeletal Mesh
    Socket Sampling Mode=::NewEnumerator5
   ) {module=SkeletalMeshLocation001}
   &ParticleState()
   &GravityForce(Gravity=<0 0 -200>)
   &ScaleSpriteSize(
    Uniform Curve Scale=0.7
    Uniform Curve Sprite Scale=DataInterface() {propertyValues {Curve {
       defaultValue=0
       keys [
        {arriveTangent=1.0437, interpMode=RCIM_Cubic, leaveTangent=1.0437, tangentMode=RCTM_User, time=0, value=0.3}
        {arriveTangent=0.5174, interpMode=RCIM_Cubic, leaveTangent=0.5174, time=0.3777, value=0.6186}
        {interpMode=RCIM_Cubic, tangentMode=RCTM_User, time=0.9843, value=0.8093}
       ]
      }}}
   )
   &SolveForcesAndVelocity(Clamp Velocity=false, Limit Acceleration=false)
   &DynamicMaterialParameters(
    Index 0 Param 1=&RandomRangeFloat(Minimum=0, Maximum=1)
    Index 0 Param 2=1
    Index 0 Param 3=&FloatFromCurve(FloatCurve=DataInterface() {propertyValues {Curve {keys [{arriveTangent=1.4668, interpMode=RCIM_Cubic, leaveTangent=1.4668, tangentMode=RCTM_User, time=0, value=0}, <at 1 1 RCIM_Cubic>]}}})
    Index 0 Param 4=&FloatFromCurve(FloatCurve=DataInterface() {propertyValues {Curve {keys [<at 0 0>,<at 0.3894 0.7206>,<at 1 1>]}}})
    Module.Index0Float0Write=true
    Module.Index0Float1Write=true
    Module.Index0Float2Write=true
    Module.Index0Float3Write=true
   )
  ]
  rendererClasses [&NiagaraSpriteRendererProperties]
  renderers [{rendererClass=&NiagaraSpriteRendererProperties}]
 }
 Intro {
  emitterSpawnScript [
   &InitializeAudioPlayer(Audio Player=DataInterface() {propertyValues {
      Attenuation=null
      Concurrency=null
      MaxPlaysPerTick=10
      SoundToPlay=/CosmeticCompanions/Assets/Audio_Shared/Interaction/Petting/SC_Mimosa_Interaction_React_Default
      bAllowLoopingOneShotSounds=false
      bLimitPlaysPerTick=true
      bOnlyActiveDuringGameplay=false
      bStopWhenComponentIsDestroyed=true
      ConfigurationUserParameter {parameter {name=None, typeDefHandle {registeredTypeIndex=1}, varData []}}
      ParameterNames []
     }})
   /Niagara/Modules/Audio/PlayAudio() {module=PlayAudio001}
   /SetVariables(Emitter.RotationRandom:NiagaraFloat=&MakeCustomFloatFromBool(Bool=&RandomBool(), True Float=1, False Float=-1)) {isSetParameters=true}
  ]
  emitterUpdateScript [
   &EmitterState()
   &SpawnBurst_Instantaneous(Use Spawn Probability=false, UseLoopCountLimit=false)
   &SpawnBurst_Instantaneous(Spawn Group=1, Use Spawn Probability=false, UseLoopCountLimit=false) {module=SpawnBurst_Instantaneous001}
  ]
  particleSpawnScript [
   /SetVariables(Particles.VisibilityTag=&SpawnGroup()) {isSetParameters=true}
   &InitializeParticle(
    Color=&MaskLinearColorBySpawnGroup(Masked Linear Color=<rgba 2 2 2 1>, Passthrough Linear Color=<rgba 0 0 0 1>, Spawn Group Mask=1)
    Lifetime=$System.IntroLength
    Position Offset=$System.BurstPosition
    Sprite Rotation Mode=::NewEnumerator0
    Sprite Size Mode=::NewEnumerator3
    Uniform Sprite Size=&FN_FrontEndFloatModifier(Gameplay=30, Front End=19.5)
    UsePositionOffset=true
    Write Material Random=true
   )
  ]
  particleUpdateScript [
   &ParticleState()
   &ScaleSpriteSize(
    Scale Sprite Size Mode=::NewEnumerator0
    Uniform Scale Factor=&FloatFromCurve(
     FloatCurve=DataInterface() {propertyValues {Curve {keys [
         {arriveTangent=1.2333, interpMode=RCIM_Cubic, leaveTangent=1.2333, tangentMode=RCTM_User, time=0, value=0.2}
         {arriveTangent=5.6726, interpMode=RCIM_Cubic, leaveTangent=5.6726, time=0.0414, value=0.2829}
         {arriveTangent=4.8098, interpMode=RCIM_Cubic, leaveTangent=4.8098, time=0.0947, value=0.7374}
         {arriveTangent=0.1086, interpMode=RCIM_Cubic, leaveTangent=0.1086, tangentMode=RCTM_User, time=0.1905, value=1}
         {arriveTangent=0.8399, interpMode=RCIM_Cubic, leaveTangent=0.8399, tangentMode=RCTM_User, time=0.4866, value=1.1757}
         {arriveTangent=0.7883, interpMode=RCIM_Cubic, leaveTangent=0.7883, time=0.6798, value=1.2763}
         {arriveTangent=0.1199, interpMode=RCIM_Constant, tangentMode=RCTM_User, time=0.8075, value=1.4287}
         <at 0.8233 1.6809>
         <at 0.9046 1.3243 RCIM_Constant>
         <at 0.9151 2.0277>
         <at 0.9989 0.6118 RCIM_Constant>
         {arriveTangent=-4.0496, interpMode=RCIM_Cubic, leaveTangent=-4.0496, time=1.0073, value=0.4151}
         {arriveTangent=-1.5827, interpMode=RCIM_Cubic, leaveTangent=-1.5827, tangentMode=RCTM_User, time=1.15, value=0}
        ]}}}
     CurveIndex=$Particles.Age
     Scale Curve=0.75
    )
   )
   &DynamicMaterialParameters(
    Index 0 Param 2=1
    Module.Index0Float0Write=true
    Module.Index0Float1Write=true
    Module.Index0Float2Write=true
    Module.Index0Float3Write=true
   )
   /SetVariables(Particles.HeartWhite_DMP:Vector4f=&MakeVector4(
     X=&FloatFromCurve(
      FloatCurve=DataInterface() {propertyValues {Curve {keys [
          {arriveTangent=0.1332, interpMode=RCIM_Cubic, leaveTangent=0.1332, tangentMode=RCTM_User, time=0.1, value=0}
          {arriveTangent=1.3071, interpMode=RCIM_Cubic, leaveTangent=1.3071, time=0.3985, value=0.1647}
          {arriveTangent=1.611, interpMode=RCIM_Cubic, leaveTangent=1.611, tangentMode=RCTM_User, time=0.7313, value=0.8252}
          <at 0.8984 1 RCIM_Cubic>
         ]}}}
      CurveIndex=$Particles.Age
      Scale Curve=0.6016
     )
     Y=0.026
     Z=0
     W=0
    )) {isSetParameters=true}
  ]
  rendererClasses [&NiagaraSpriteRendererProperties]
  renderers [
   {rendererClass=&NiagaraSpriteRendererProperties}
   {rendererClass=&NiagaraSpriteRendererProperties, rendererIndex=1}
  ]
 }
}
summary {
 systemName=NS_ReactFX_ComicTabby
 emitters [
  {emitterName=Drops, rendererClasses [&NiagaraSpriteRendererProperties]}
  {emitterName=InkSplat, rendererClasses [&NiagaraSpriteRendererProperties]}
  {emitterName=Intro, rendererClasses [&NiagaraSpriteRendererProperties]}
 ]
 userVariables [
  User.bIsFrontEnd {type=NiagaraBool}
  User.bIsFrontEndPreview {type=NiagaraBool}
  User.Intensity {type=NiagaraFloat, defaultValue=0.68, description=""}
 ]
}
