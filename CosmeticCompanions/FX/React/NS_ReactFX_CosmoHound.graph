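# Niagara reaction effect for the CosmoHound cosmetic companion: timed bursts of mesh cubes and lightning with glow.
# The Cubes emitter combines both, using spawn groups and visibility tags across four mesh renderers.
# An emitter-age switch changes cube spread, speed and scale, while point attraction pulls particles in and kills them.
# System variables drive timing, offsets and front-end scale; separate cube, light and lightning emitters shown here are disabled.
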NiagaraSystem /CosmeticCompanions/FX/React/NS_ReactFX_CosmoHound
dependencies {
 usedDataInterfaces [
  &NiagaraDataInterfaceAudioPlayer
  &NiagaraDataInterfaceArrayFloat
  &NiagaraDataInterfaceCurve
  &NiagaraDataInterfaceVectorCurve
  &NiagaraDataInterfaceColorCurve
  &NiagaraDataInterfaceVector2DCurve
  &NiagaraDataInterfaceMeshRendererInfo
  &NiagaraDataInterfaceCamera
  &NiagaraDataInterfacePlatformSet
 ]
 usedDynamicInputs [&IsPlatformSetActive]
 usedModules [
  &self:SystemSpawnScript.SystemScriptSource.SystemScriptGraph.NiagaraNodeAssignment_9.SetVariables_#1
  &self:SystemSpawnScript.SystemScriptSource.SystemScriptGraph.NiagaraNodeAssignment_3.SetVariables_#2
  &SystemState
  &Timeline
  &self:SystemSpawnScript.SystemScriptSource.SystemScriptGraph.NiagaraNodeAssignment_12.SetVariables_#3
  &self:SystemSpawnScript.SystemScriptSource.SystemScriptGraph.NiagaraNodeAssignment_0.SetVariables_#4
  &self:SystemSpawnScript.SystemScriptSource.SystemScriptGraph.NiagaraNodeAssignment_1.SetVariables_#5
 ]
}
emitters {
 Cubes {
  emitterUpdateScript [
   &EmitterState(Life Cycle Mode=::NewEnumerator1, Loop Behavior=::NewEnumerator1, Loop Duration=1.5, UseLoopDelay=false)
   &SpawnBurst_Instantaneous(Spawn Count=15, Spawn Time=0.5, Use Spawn Probability=false, UseLoopCountLimit=false)
   &SpawnBurst_Instantaneous(Spawn Count=25, Spawn Time=$System.PopTiming, Use Spawn Probability=false, UseLoopCountLimit=false) {module=SpawnBurst_Instantaneous001}
   &SpawnBurst_Instantaneous(Spawn Count=7, Spawn Group=1, Spawn Time=1, Use Spawn Probability=false, UseLoopCountLimit=false) {module=SpawnBurst_Instantaneous002}
   &SpawnBurst_Instantaneous(Spawn Count=3, Spawn Group=2, Spawn Time=0.5, Use Spawn Probability=false, UseLoopCountLimit=false) {module=SpawnBurst_Instantaneous003}
   &SpawnRate(
    SpawnGroup=1
    SpawnRate=&FloatFromCurve(
     FloatCurve=DataInterface() {propertyValues {Curve {
        defaultValue=1
        keys [
         <at 0 0 RCIM_Cubic>
         <at 0.1196 0.6932 RCIM_Cubic>
         {arriveTangent=1.1125, arriveTangentWeight=0.0646, interpMode=RCIM_Cubic, leaveTangent=1.1125, tangentMode=RCTM_User, time=0.3135, value=0}
         {arriveTangent=1.2369, interpMode=RCIM_Cubic, leaveTangent=1.2369, time=0.4343, value=3.4471}
         {
          arriveTangent=0.1357
          arriveTangentWeight=0.086
          interpMode=RCIM_Cubic
          leaveTangent=0.1357
          tangentMode=RCTM_User
          time=0.6922
          value=0.4684
         }
         {arriveTangent=-1.5218, interpMode=RCIM_Cubic, leaveTangent=-1.5218, time=0.8588, value=1}
         <at 1 0 RCIM_Cubic>
        ]
       }}}
     CurveIndex=$Emitter.NormalizedLoopAge
    )
    Use Spawn Probability=false
   )
   /SetVariables(Emitter.OutTiming:NiagaraBool=&SetBoolByFloatComparison(A=$Emitter.Age, B=0.8, Comparison Type=::NewEnumerator2)) {isSetParameters=true, module=SetVariables_#6}
  ]
  particleSpawnScript [
   &PartitionParticles(Number Of Partitions=3, Partition Mode:ENiagara_PartitionMode_v2=::NewEnumerator5)
   &InitializeParticle(
    AdjustAlpha=false
    AdjustHue=false
    AdjustSaturation=false
    AdjustValue=true
    Color=&MaskLinearColor(Mask Source=0, Mask Parameter=?, Mask=0, Masked Linear Color=<rgba 1 1 1 1>, Passthrough Linear Color=<rgba 2 0.1311 0 1>)
    Color Mode=::NewEnumerator3
    Lifetime=&MaskFloat(
     Mask Source=0
     Mask Parameter=?
     Mask=0
     Masked Float=0.44
     Passthrough Float=&RandomRangeFloat(Minimum=0.2, Maximum=0.3)
    )
    Mesh Scale Mode=::NewEnumerator3
    Mesh Uniform Scale=&FloatFromCurve(
     FloatCurve=DataInterface() {propertyValues {Curve {
        defaultValue=0.5
        keys [
         {arriveTangent=1.3886, interpMode=RCIM_Cubic, leaveTangent=1.3886, tangentMode=RCTM_User, time=0.2, value=0.5}
         {arriveTangentWeight=0.1667, interpMode=RCIM_Cubic, tangentMode=RCTM_User, time=0.6, value=1}
        ]
       }}}
     CurveIndex=$Emitter.NormalizedLoopAge
     Scale Curve=0.5
    )
    Position Offset=&MaskVector(Mask Source=0, Mask Parameter=?, Mask=0, Masked Vector=<0 0 5>, Passthrough Vector=$System.BurstPosition)
    Saturation Range:Vector2f=<1 1.5>
    UsePositionOffset=true
    Write Material Random=true
   )
   /SetVariables(
    Particles.CenterGradient:NiagaraFloat=&RandomRangeFloat(Minimum=0, Maximum=1)
    Particles.SideChoice:NiagaraInt32=&MakeIntFromFloat(Float=&FloatFromCurve(
      FloatCurve=DataInterface() {propertyValues {Curve {defaultValue=1, keys [<at 0 0 RCIM_Constant>,<at 0.5102 1 RCIM_Constant>,<at 1 2 RCIM_Constant>]}}}
      CurveIndex=&RandomRangeFloat(Minimum=0, Maximum=1)
     ))
   ) {isSetParameters=true, module=SetVariables_#7}
   /SetVariables(
    Particles.LightningIndex:NiagaraInt32=&RandomRangeInt(Minimum=0, Maximum=3)
    Particles.LightningLength:NiagaraFloat=/Niagara/DynamicInputs/Helpers/DistributionCurveFloat(
     Minimum:NiagaraFloat=0.3
     Maximum:NiagaraFloat=0.6
     Distribution Curve:NiagaraDataInterfaceCurve=DataInterface() {propertyValues {Curve {defaultValue=0, keys [<at 0.3 0.2 RCIM_Constant>, <at 0.6 0.3 RCIM_Constant>, {tangentMode=RCTM_User, time=0.85, value=1}]}}}
     Curve Parameter Type:NiagaraInt32=0
     "Curve Parameter Type:":NiagaraInt32=0
     Curve Parameter:NiagaraFloat=?
     Evaluation Type:ENiagaraRandomnessEvaluation=::NewEnumerator0
     Recalculate Random Each Loop:NiagaraBool=true
     Randomness Mode:ENiagaraRandomnessMode=::NewEnumerator0
     Random Seed:NiagaraInt32=0
     Fixed Random Seed:NiagaraBool=false
    )
    Particles.MirrorLightning:NiagaraBool=&RandomBool()
    Particles.Random=&RandomRangeFloat(Minimum=0, Maximum=1)
    Particles.VisibilityTag=&SpawnGroup()
   ) {isSetParameters=true}
   &ShapeLocation(
    MaskSpawnGroup=false
    Non Uniform Scale=&Multiply_VectorByFloat(Vector=<1 1 1>, Float=$System.FE_Scale)
    Override Seed=false
    Sphere Radius=&MakeCustomFloatFromBool(Bool=$Emitter.OutTiming, True Float=5, False Float=25)
    UseOwnerScale=true
   )
   &AddVelocity(
    Module.Use Falloff Distance=false
    Origin Offset=&MakeVector(
     X=0
     Y=0
     Z=&Add_Float(A=5, B=&MakeFloatFromVector(Vector=<0 0 0>, Channel=::NewEnumerator2))
    )
    Override Seed=false
    Velocity Mode=::NewEnumerator1
    Velocity Speed=&Multiply_Float(
     A=&MakeCustomFloatFromBool(Bool=$Emitter.OutTiming, True Float=&RandomRangeFloat(Minimum=300, Maximum=400), False Float=25)
     B=$System.FE_Scale
    )
   )
   &ApplyInitialForces()
   &SubUVAnimation(
    Mesh Renderer=DataInterface() {propertyValues {MeshRenderer=&self:Cubes_0.NiagaraMeshRendererProperties_0}}
    Output Alternate Renderer Binding=true
    Renderer Type:ENiagaraMeshOrSprite=::NewEnumerator1
    UseEndFrame=false
    UseStartFrame=false
   ) {enabled=false}
   &DynamicMaterialParameters(
    Index 0 Param 1=1
    Index 0 Param 2=0.2
    Module.Index0Float0Write=true
    Module.Index0Float1Write=true
    Module.Index0Float2Write=false
    Module.Index0Float3Write=false
   )
   &InitialMeshOrientation(
    Mesh Orientation Mode=::NewEnumerator2
    Orientation Vector=&MaskVector(
     Mask Source=0
     Mask Parameter=?
     Mask=0
     Masked Vector=<1 0 0>
     Passthrough Vector=&FN_FrontEndVector3Modifier(Gameplay=<-1 0 0>, Front End=&RandomVector(Vector Scale=1))
    )
    Rotation=&MaskVector(
     Mask Source=0
     Mask Parameter=?
     Mask=0
     Masked Vector=&MakeVector(X=0, Y=0, Z=&RandomRangeFloat(Minimum=0, Maximum=1))
     Passthrough Vector=&MakeVector(
      X=&Add_Float(
       A=&FloatFromCurve(
        FloatCurve=DataInterface() {propertyValues {Curve {keys [<at 0 1>,<at 1 0>]}}}
        CurveIndex=&ReturnNormalizedExecIndex()
       )
       B=&RandomRangeFloat(Minimum=-0.1, Maximum=0.1)
      )
      Y=&RandomRangeFloat(Minimum=-0.3, Maximum=0.3)
      Z=0
     )
    )
    Use Rotation Vector=true
   )
  ]
  particleUpdateScript [
   &ParticleState()
   &ScaleMeshSize(Scale Factor=&Multiply_VectorByFloat(
     Vector=<1 1 1>
     Float=&MakeCustomFloatFromBool(
      Bool=$Emitter.OutTiming
      True Float=&FloatFromCurve(
       FloatCurve=DataInterface() {propertyValues {Curve {keys [
           {arriveTangent=15.7898, interpMode=RCIM_Cubic, leaveTangent=15.7898, tangentMode=RCTM_User, time=0, value=0}
           {interpMode=RCIM_Cubic, tangentMode=RCTM_User, time=0.1, value=1.1612}
           {arriveTangent=-1.5494, interpMode=RCIM_Cubic, leaveTangent=-1.5494, tangentMode=RCTM_User, time=0.6, value=0.6612}
           {interpMode=RCIM_Cubic, tangentMode=RCTM_User, time=1, value=0.1}
          ]}}}
       Scale Curve=$System.FE_Scale
      )
      False Float=&FloatFromCurve(
       FloatCurve=DataInterface() {propertyValues {Curve {keys [
           {arriveTangent=37.1186, interpMode=RCIM_Cubic, leaveTangent=37.1186, tangentMode=RCTM_User, time=0, value=0}
           {interpMode=RCIM_Cubic, tangentMode=RCTM_User, time=0.2, value=1}
           {interpMode=RCIM_Cubic, tangentMode=RCTM_User, time=1, value=1}
          ]}}}
       Scale Curve=$System.FE_Scale
      )
     )
    ))
   &PointAttractionForce(
    Attraction Strength=&FloatFromCurve(
     FloatCurve=DataInterface() {propertyValues {Curve {keys [
         {interpMode=RCIM_Cubic, tangentMode=RCTM_User, time=0.2, value=1}
         {arriveTangent=5, interpMode=RCIM_Cubic, leaveTangent=5, time=0.3283, value=3.1576}
         <at 1 5>
        ]}}}
     CurveIndex=$Emitter.NormalizedLoopAge
     Scale Curve=&Multiply_Float(A=100, B=$System.FE_Scale)
    )
    Attractor Position Offset:Vector3f=&AddVectorToPosition(Position=Linked() {linkedVariable {name=System.IntroPositionOffset, type=Vector3f}}, Vector=<0 0 0>)
    Falloff Exponent=0.45
    Kill Radius=3
    Kill Within Radius=true
    Module.Use Falloff=true
    UseAttractorOffset=true
   )
   &SolveForcesAndVelocity(Clamp Velocity=false, Limit Acceleration=false)
   &ScaleColor(
    Scale Alpha=&MaskFloat(
     Mask Source=0
     Mask Parameter=?
     Mask=0
     Masked Float=1
     Passthrough Float=&Multiply_Float(
      A=10
      B=&FloatFromCurve(FloatCurve=DataInterface() {propertyValues {Curve {keys [<at 0 0>,<at 0.1 0>,<at 0.2 0.8>,<at 1 0>]}}})
     )
    )
    Scale RGB=&MaskVector(
     Mask Source=0
     Mask Parameter=?
     Mask=0
     Masked Vector=&Multiply_VectorByFloat(
      Vector=<1 1 1>
      Float=&FloatFromCurve(
       FloatCurve=DataInterface() {propertyValues {Curve {
          defaultValue=1
          keys [
           {interpMode=RCIM_Cubic, tangentMode=RCTM_User, time=0, value=1}
           <at 0.102 0.456 RCIM_Cubic>
           {interpMode=RCIM_Cubic, tangentMode=RCTM_User, time=0.2559, value=1}
           <at 0.4642 0.5 RCIM_Cubic>
           <at 1 1 RCIM_Cubic>
          ]
         }}}
       Scale Curve=&FN_FrontEndFloatModifier(Gameplay=1, Front End=0.7)
      )
     )
     Passthrough Vector=&Multiply_VectorByFloat(
      Vector=&Multiply_VectorByFloat(Vector=<1 1 1>, Float=&FN_FrontEndFloatModifier(Gameplay=0.5, Front End=0.5))
      Float=&Multiply_Float(
       A=&RandomRangeFloat(Minimum=0.35, Maximum=1, Evaluation Type=::NewEnumerator1)
       B=&Multiply_Float(A=&FNTimeOfDayScalar(Night Scale=0.2, Morning Scale=1, Evening Scale=1), B=10)
      )
     )
    )
    ScaleA=true
    ScaleRGB=true
   )
   &OffsetPosition(Position Offset=$System.IntroPositionOffset)
   &DynamicMaterialParameters(
    Index 0 Param 1=&FloatFromCurve(FloatCurve=DataInterface() {propertyValues {Curve {keys [<at 0 0>,<at 1 4>]}}})
    Index 0 Param 2=1
    Index 0 Param 3=&Waveform("Waveform [1]"=::NewEnumerator0, "[1] Frequency"=0.4, "[1] Amplitude Min/Max"=<1 2>)
    Index 2 Param 1=&FloatFromCurve(FloatCurve=DataInterface() {propertyValues {Curve {keys [<at 0 1>,<at 1 0.3>]}}})
    Module.Index0Float0Write=true
    Module.Index0Float1Write=true
    Module.Index0Float2Write=true
    Module.Index0Float3Write=true
    Module.Index2Float0Write=true
    Module.Index2Float1Write=true
    Module.Index2Float2Write=true
    Module.Index2Float3Write=true
    Write Parameter Index 2=::NewEnumerator1
   ) {module=DynamicMaterialParameters001}
   &FN_Mesh_Attributes(
    Create Scale Attribute=true
    Scale=&Multiply_Vector(
     A=&Multiply_VectorByFloat(
      Vector=&MakeCustomVectorFromBool(Bool=$Particles.Initial.MirrorLightning, True Vector=<1 -1 -1>, False Vector=<1 1 1>)
      Float=$Particles.Initial.LightningLength
     )
     B=&Multiply_VectorByFloat(
      Vector=&MakeVector(
       X=&FloatFromCurve(FloatCurve=DataInterface() {propertyValues {Curve {
           defaultValue=1
           keys [
            {arriveTangent=0.8401, interpMode=RCIM_Cubic, leaveTangent=0.8401, tangentMode=RCTM_User, time=0, value=0.8}
            {interpMode=RCIM_Cubic, tangentMode=RCTM_User, time=1, value=1}
           ]
          }}})
       Y=1
       Z=1
      )
      Float=$System.FE_Scale
     )
    )
   )
   /SetVariables(
    Particles.DynamicParametersAlt:Vector4f=&MakeVector4(
     X=1
     Y=0.7
     Z=&FloatFromCurve(FloatCurve=DataInterface() {propertyValues {Curve {keys [
          {arriveTangent=4.0072, interpMode=RCIM_Cubic, leaveTangent=4.0072, tangentMode=RCTM_User, time=0, value=0}
          {arriveTangent=2.7083, interpMode=RCIM_Cubic, leaveTangent=2.7083, time=0.2543, value=0.6225}
          {arriveTangent=1.8473, interpMode=RCIM_Cubic, leaveTangent=1.8473, time=0.6, value=1.625}
          {arriveTangent=1.6993, interpMode=RCIM_Cubic, leaveTangent=1.6993, tangentMode=RCTM_User, time=1, value=2}
         ]}}})
     W=13
    )
    Particles.DynamicParametersAlt2:Vector4f=&MakeVector4(
     X=&Multiply_Float(A=$Particles.Initial.LightningLength, B=&RandomRangeFloat(Minimum=0.5, Maximum=1.5))
     Y=&UniformAOrBFloat(A=0.5, B=1.5, Evaluation Type=::NewEnumerator0, Recalculate Random Each Loop=true, Override Seed=false)
     Z=&FloatFromCurve(FloatCurve=DataInterface() {propertyValues {Curve {
         defaultValue=1
         keys [
          {arriveTangent=-2.478, interpMode=RCIM_Cubic, leaveTangent=-2.478, tangentMode=RCTM_User, time=0, value=1}
          {interpMode=RCIM_Cubic, tangentMode=RCTM_User, time=1, value=0}
         ]
        }}})
     W=0
    )
   ) {isSetParameters=true}
   &FN_Mesh_Attributes(
    Color=&MultiplyLinearColor(
     Linear Color=&RandomRangeLinearColor(Minimum=<rgba 5 1 5 1>, Maximum=<rgba 5 1 3.6667 1>)
     Scale Factor=&MakeVector4FromVectorAndFloat(
      "Vector (RGB)"=&VectorFromFloat(Value=&RandomRangeFloat(Minimum=0, Maximum=2, Evaluation Type=::NewEnumerator1))
      "Float (Alpha)"=&FloatFromCurve(FloatCurve=DataInterface() {propertyValues {Curve {keys [<at 0.7 1>,<at 1 0>]}}})
     )
    )
    Create Color Attribute=true
    Create Orientation Attribute=true
    Create Scale Attribute=true
    Orientation:Quat4f=&MakeQuaternion(
     Quaternion From=1
     Axis=&RandomRangeVector(Minimum=<-0.3 -0.3 -1>, Maximum=<0.3 0.3 1>)
     Angle=&RandomRangeFloat(Minimum=0, Maximum=360)
     Yaw=&RandomRangeFloat(Minimum=-0.1, Maximum=0.1)
     Pitch=&RandomRangeFloat(Minimum=0.05, Maximum=0.45)
     Roll=&RandomRangeFloat(Minimum=-1, Maximum=1)
     From=?
     To=?
     Angle Type=::NewEnumerator1
     Coordinate Space=::Local
     Rotation Matrix=Linked() {linkedVariable {name=Engine.Owner.SystemLocalToWorldNoScale, type=NiagaraInt32, defaultValue=~, description=~}}
     Normalize Vectors=false
    )
    Scale=&Multiply_VectorByFloat(
     Vector=&VectorFromFloat(Value=&Multiply_Float(A=0.3, B=0.6))
     Float=&FloatFromCurve(
      FloatCurve=DataInterface() {propertyValues {Curve {keys [
          {arriveTangent=3.2151, interpMode=RCIM_Cubic, leaveTangent=3.2151, tangentMode=RCTM_User, time=0, value=0.6}
          {arriveTangent=0.4, interpMode=RCIM_Cubic, leaveTangent=0.4, time=0.5973, value=0.9154}
          {arriveTangent=0.0224, interpMode=RCIM_Cubic, leaveTangent=0.0224, tangentMode=RCTM_User, time=1, value=1}
         ]}}}
      Scale Curve=$System.FE_Scale
     )
    )
    Write to Module Namespace=true
   ) {module=FN_Mesh_Attributes002}
   /SetVariables(
    Particles.DynamicParametersArc1:Vector4f=&MakeVector4(X=0, Y=0, Z=0, W=13)
    Particles.DynamicParametersArc2:Vector4f=&MakeVector4(
     X=&FloatFromCurve(FloatCurve=DataInterface() {propertyValues {Curve {keys [<at 0 0.5>,<at 1 1>]}}})
     Y=3
     Z=&FloatFromCurve(FloatCurve=DataInterface() {propertyValues {Curve {keys [{interpMode=RCIM_Cubic, tangentMode=RCTM_Break, time=0.5, value=-0.5}, <at 1 0.1>]}}})
     W=3
    )
    Particles.DynamicParametersArc3:Vector4f=&MakeVector4(
     X=&Multiply_Float(
      A=&FloatFromCurve(FloatCurve=DataInterface() {propertyValues {Curve {keys [{interpMode=RCIM_Cubic, leaveTangent=8.1085, tangentMode=RCTM_Break, time=0.8568, value=1}, {tangentMode=RCTM_User, time=1, value=0}]}}})
      B=&Divide_Float(
       A=&MakeFloatFromVector(Vector=$Particles.FN_Mesh_Attributes002.Scale, Channel=::NewEnumerator0)
       B=&RandomRangeFloat(Minimum=50, Maximum=100)
      )
     )
     Y=&FloatFromCurve(
      FloatCurve=DataInterface() {propertyValues {Curve {keys [
          {interpMode=RCIM_Cubic, tangentMode=RCTM_Break, time=0.7, value=0}
          {arriveTangent=-0.0859, interpMode=RCIM_Cubic, tangentMode=RCTM_Break, time=1, value=-1}
         ]}}}
      Scale Curve=50
     )
     Z=&Waveform(
      Global Input Position=&RandomRangeFloat(Minimum=0, Maximum=15)
      Global Phase=$Particles.Age
      "Waveform [1]"=::NewEnumerator12
      "[1] Phase Scale"=2
      "[1] Frequency"=1
      "[1] Amplitude Min/Max"=<0 1>
     )
     W=25
    )
    Particles.DynamicParametersArc4:Vector4f=&MakeVector4(
     X=&FloatFromCurve(
      FloatCurve=DataInterface() {propertyValues {Curve {
         defaultValue=1
         keys [
          {arriveTangent=2.157, interpMode=RCIM_Cubic, leaveTangent=2.157, tangentMode=RCTM_User, time=0, value=0}
          {interpMode=RCIM_Cubic, tangentMode=RCTM_User, time=1, value=1}
         ]
        }}}
      Scale Curve=10
     )
     Y=0.1
     Z=&FloatFromCurve(FloatCurve=DataInterface() {propertyValues {Curve {keys [
          {interpMode=RCIM_Cubic, leaveTangent=1.1034, tangentMode=RCTM_Break, time=0, value=-1}
          {arriveTangent=2.2399, tangentMode=RCTM_User, time=0.5, value=0.125}
         ]}}})
     W=&FloatFromCurve(FloatCurve=DataInterface() {propertyValues {Curve {keys [
          {interpMode=RCIM_Cubic, leaveTangent=1.8089, tangentMode=RCTM_Break, time=0.3, value=-0.5}
          {arriveTangent=2.0757, interpMode=RCIM_Cubic, tangentMode=RCTM_Break, time=0.9, value=1}
         ]}}})
    )
   ) {isSetParameters=true}
   /SetVariables(Particles.Lightningglowcolor:LinearColor=&MakeLinearColorFromVectorAndFloat(
     "Vector (RGB)"=&MakeVectorFromLinearColorRGB(Color=&MultiplyLinearColor(
       Linear Color=<rgba 5 2.6797 0.3594 1>
       Scale Factor=&MakeVector4_FromFloat(Float=&RandomRangeFloat(Minimum=0.6, Maximum=0.8, Evaluation Type=::NewEnumerator1))
      ))
     "Float (Alpha)"=&Multiply_Float(
      A=&FloatFromCurve(FloatCurve=DataInterface() {propertyValues {Curve {keys [<at 0 1 RCIM_Cubic>, {interpMode=RCIM_Cubic, leaveTangent=-3.6307, time=0.9, value=0}]}}})
      B=1
     )
    )) {isSetParameters=true}
   /SetVariables(Particles.Glow Dy02:Vector4f=&MakeVector4(
     X=&FloatFromCurve(FloatCurve=DataInterface() {propertyValues {Curve {keys [{tangentMode=RCTM_User, time=0, value=0.2}, {tangentMode=RCTM_User, time=0.5, value=1}]}}})
     Y=0.7
     Z=0
     W=13
    )) {isSetParameters=true}
   /SetVariables(Particles.Glow Dy03:Vector4f=&MakeVector4(
     X=&FloatFromCurve(
      FloatCurve=DataInterface() {propertyValues {Curve {keys [<at 0 1>,<at 1 0.5>]}}}
      Scale Curve=&FN_FrontEndFloatModifier(Gameplay=1.5, Front End=1)
     )
     Y=$Particles.Initial.Random
     Z=1
     W=1
    )) {isSetParameters=true}
  ]
  rendererClasses [&NiagaraMeshRendererProperties]
  renderers [
   {rendererClass=&NiagaraMeshRendererProperties, rendererIndex=0}
   {rendererClass=&NiagaraMeshRendererProperties, rendererIndex=1}
   {rendererClass=&NiagaraMeshRendererProperties, rendererIndex=2}
   {rendererClass=&NiagaraMeshRendererProperties, rendererIndex=3}
  ]
 }
 Cubes001 {
  bEnabled=false
  emitterUpdateScript [
   &EmitterState(Life Cycle Mode=::NewEnumerator1, Loop Behavior=::NewEnumerator1, Loop Duration=1.5, UseLoopDelay=false)
   &SpawnBurst_Instantaneous(Spawn Count=15, Spawn Time=0.5, Use Spawn Probability=false, UseLoopCountLimit=false)
   &SpawnBurst_Instantaneous(Spawn Count=25, Spawn Time=$System.PopTiming, Use Spawn Probability=false, UseLoopCountLimit=false) {module=SpawnBurst_Instantaneous001}
   /SetVariables(Emitter.OutTiming:NiagaraBool=&SetBoolByFloatComparison(A=$Emitter.Age, B=0.8, Comparison Type=::NewEnumerator2)) {isSetParameters=true, module=SetVariables_#6}
  ]
  particleSpawnScript [
   &InitializeParticle(
    AdjustAlpha=false
    AdjustHue=false
    AdjustSaturation=false
    AdjustValue=true
    Color Mode=::NewEnumerator3
    Lifetime=0.4
    Mesh Scale Mode=::NewEnumerator3
    Mesh Uniform Scale=&FloatFromCurve(
     FloatCurve=DataInterface() {propertyValues {Curve {
        defaultValue=0.5
        keys [
         {arriveTangent=1.3886, interpMode=RCIM_Cubic, leaveTangent=1.3886, tangentMode=RCTM_User, time=0.2, value=0.5}
         {arriveTangentWeight=0.1667, interpMode=RCIM_Cubic, tangentMode=RCTM_User, time=0.6, value=1}
        ]
       }}}
     CurveIndex=$Emitter.NormalizedLoopAge
     Scale Curve=0.5
    )
    Position Offset=<0 0 5>
    Saturation Range:Vector2f=<1 1.5>
    UsePositionOffset=true
    Write Material Random=true
   )
   /SetVariables(
    Particles.CenterGradient:NiagaraFloat=&RandomRangeFloat(Minimum=0, Maximum=1)
    Particles.SideChoice:NiagaraInt32=&MakeIntFromFloat(Float=&FloatFromCurve(
      FloatCurve=DataInterface() {propertyValues {Curve {defaultValue=1, keys [<at 0 0 RCIM_Constant>,<at 0.5102 1 RCIM_Constant>,<at 1 2 RCIM_Constant>]}}}
      CurveIndex=&RandomRangeFloat(Minimum=0, Maximum=1)
     ))
   ) {isSetParameters=true, module=SetVariables_#7}
   &ShapeLocation(
    MaskSpawnGroup=false
    Non Uniform Scale=&Multiply_VectorByFloat(Vector=<1 1 1>, Float=$System.FE_Scale)
    Override Seed=false
    Sphere Radius=&MakeCustomFloatFromBool(Bool=$Emitter.OutTiming, True Float=5, False Float=25)
    UseOwnerScale=true
   )
   &AddVelocity(
    Module.Use Falloff Distance=false
    Origin Offset=&MakeVector(
     X=0
     Y=0
     Z=&Add_Float(A=5, B=&MakeFloatFromVector(Vector=<0 0 0>, Channel=::NewEnumerator2))
    )
    Override Seed=false
    Velocity Mode=::NewEnumerator1
    Velocity Speed=&Multiply_Float(
     A=&MakeCustomFloatFromBool(Bool=$Emitter.OutTiming, True Float=&RandomRangeFloat(Minimum=300, Maximum=400), False Float=25)
     B=$System.FE_Scale
    )
   )
   &ApplyInitialForces()
   &SubUVAnimation(
    Mesh Renderer=DataInterface() {propertyValues {MeshRenderer=&self:Cubes001_0.NiagaraMeshRendererProperties_0}}
    Output Alternate Renderer Binding=true
    Renderer Type:ENiagaraMeshOrSprite=::NewEnumerator1
    UseEndFrame=false
    UseStartFrame=false
   )
   &DynamicMaterialParameters(
    Index 0 Param 1=1
    Index 0 Param 2=0.2
    Module.Index0Float0Write=true
    Module.Index0Float1Write=true
    Module.Index0Float2Write=false
    Module.Index0Float3Write=false
   )
   &InitialMeshOrientation(
    Mesh Orientation Mode=::NewEnumerator2
    Rotation=&MakeVector(X=0, Y=0, Z=&RandomRangeFloat(Minimum=0, Maximum=1))
    Use Rotation Vector=true
   )
  ]
  particleUpdateScript [
   &ParticleState()
   &ScaleMeshSize(Scale Factor=&Multiply_VectorByFloat(
     Vector=<1 1 1>
     Float=&MakeCustomFloatFromBool(
      Bool=$Emitter.OutTiming
      True Float=&FloatFromCurve(
       FloatCurve=DataInterface() {propertyValues {Curve {keys [
           {arriveTangent=15.7898, interpMode=RCIM_Cubic, leaveTangent=15.7898, tangentMode=RCTM_User, time=0, value=0}
           {interpMode=RCIM_Cubic, tangentMode=RCTM_User, time=0.1, value=1.1612}
           {arriveTangent=-1.5494, interpMode=RCIM_Cubic, leaveTangent=-1.5494, tangentMode=RCTM_User, time=0.6, value=0.6612}
           {interpMode=RCIM_Cubic, tangentMode=RCTM_User, time=1, value=0.1}
          ]}}}
       Scale Curve=$System.FE_Scale
      )
      False Float=&FloatFromCurve(
       FloatCurve=DataInterface() {propertyValues {Curve {keys [
           {arriveTangent=37.1186, interpMode=RCIM_Cubic, leaveTangent=37.1186, tangentMode=RCTM_User, time=0, value=0}
           {interpMode=RCIM_Cubic, tangentMode=RCTM_User, time=0.2, value=1}
           {interpMode=RCIM_Cubic, tangentMode=RCTM_User, time=1, value=1}
          ]}}}
       Scale Curve=$System.FE_Scale
      )
     )
    ))
   &PointAttractionForce(
    Attraction Strength=&FloatFromCurve(
     FloatCurve=DataInterface() {propertyValues {Curve {keys [
         {interpMode=RCIM_Cubic, tangentMode=RCTM_User, time=0.2, value=1}
         {arriveTangent=5, interpMode=RCIM_Cubic, leaveTangent=5, time=0.3283, value=3.1576}
         <at 1 5>
        ]}}}
     CurveIndex=$Emitter.NormalizedLoopAge
     Scale Curve=&Multiply_Float(A=100, B=$System.FE_Scale)
    )
    Attractor Position Offset:Vector3f=&AddVectorToPosition(Position=Linked() {linkedVariable {name=System.IntroPositionOffset, type=Vector3f}}, Vector=<0 0 0>)
    Falloff Exponent=0.45
    Kill Radius=3
    Kill Within Radius=true
    Module.Use Falloff=true
    UseAttractorOffset=true
   )
   &SolveForcesAndVelocity(Clamp Velocity=false, Limit Acceleration=false)
   &ScaleColor(
    Scale Alpha=&Add_Float(
     A=&FloatFromCurve(FloatCurve=DataInterface() {propertyValues {Curve {
         defaultValue=1
         keys [
          {arriveTangent=0.2063, interpMode=RCIM_Cubic, leaveTangent=0.2063, tangentMode=RCTM_User, time=0, value=0.1}
          {arriveTangent=0.4197, interpMode=RCIM_Cubic, leaveTangent=0.4197, tangentMode=RCTM_User, time=0.5178, value=5}
          {interpMode=RCIM_Cubic, tangentMode=RCTM_User, time=1, value=5}
         ]
        }}})
     B=&FloatFromCurve(
      FloatCurve=DataInterface() {propertyValues {Curve {
         defaultValue=1
         keys [
          {arriveTangent=0.0573, interpMode=RCIM_Cubic, leaveTangent=0.0573, tangentMode=RCTM_User, time=0, value=0}
          {interpMode=RCIM_Cubic, tangentMode=RCTM_User, time=0.5, value=0.5}
          {arriveTangent=-0.0685, interpMode=RCIM_Cubic, leaveTangent=-0.0685, tangentMode=RCTM_User, time=0.7, value=0.001}
         ]
        }}}
      CurveIndex=$Emitter.NormalizedLoopAge
     )
    )
    Scale RGB=&Multiply_VectorByFloat(
     Vector=<1 1 1>
     Float=&FloatFromCurve(
      FloatCurve=DataInterface() {propertyValues {Curve {
         defaultValue=1
         keys [
          {interpMode=RCIM_Cubic, tangentMode=RCTM_User, time=0, value=1}
          <at 0.102 0.456 RCIM_Cubic>
          {interpMode=RCIM_Cubic, tangentMode=RCTM_User, time=0.2559, value=1}
          <at 0.4642 0.5 RCIM_Cubic>
          <at 1 1 RCIM_Cubic>
         ]
        }}}
      Scale Curve=&FN_FrontEndFloatModifier(Gameplay=1, Front End=0.7)
     )
    )
    ScaleA=false
    ScaleRGB=true
   )
   &OffsetPosition(Position Offset=$System.IntroPositionOffset)
   &DynamicMaterialParameters(
    Index 0 Param 1=&FloatFromCurve(FloatCurve=DataInterface() {propertyValues {Curve {keys [<at 0 0>,<at 1 4>]}}})
    Index 0 Param 2=1
    Index 0 Param 3=&Waveform("Waveform [1]"=::NewEnumerator0, "[1] Frequency"=0.4, "[1] Amplitude Min/Max"=<1 2>)
    Module.Index0Float0Write=true
    Module.Index0Float1Write=true
    Module.Index0Float2Write=true
    Module.Index0Float3Write=true
   ) {module=DynamicMaterialParameters001}
  ]
  rendererClasses [&NiagaraMeshRendererProperties]
  renderers [{rendererClass=&NiagaraMeshRendererProperties, rendererIndex=0}]
 }
 Glow_And_Light {
  bEnabled=false
  emitterUpdateScript [
   &EmitterState()
   &SpawnBurst_Instantaneous(Spawn Group=1, Spawn Time=0.05, Use Spawn Probability=false, UseLoopCountLimit=false) {module=SpawnBurst_Instantaneous001}
   &SpawnBurst_Instantaneous(Spawn Time=&Subtract_Float(A=$System.IntroLength, B=0), Use Spawn Probability=false, UseLoopCountLimit=false)
  ]
  particleSpawnScript [
   &InitializeParticle(
    Color=&MaskLinearColorBySpawnGroup(Masked Linear Color=<rgba 3 3 0 0.8>, Passthrough Linear Color=<rgba 1.013 1.5 0.0391 0.5>)
    Lifetime=&MaskFloatBySpawnGroup(Masked Float=0.4, Passthrough Float=0.8)
    Position Offset=$System.BurstPosition
    Sprite Rotation Mode=::NewEnumerator0
    Sprite Size Mode=::NewEnumerator3
    Uniform Sprite Size=&Multiply_Float(
     A=&MaskFloatBySpawnGroup(Masked Float=30, Passthrough Float=30)
     B=&FN_FrontEndFloatModifier(Gameplay=1, Front End=$System.FrontendScale)
    )
    UsePositionOffset=true
    Write Material Random=true
   )
   &DynamicMaterialParameters(
    Index 0 Param 4=1
    Module.Index0Float0Write=true
    Module.Index0Float1Write=true
    Module.Index0Float2Write=true
    Module.Index0Float3Write=true
   )
   /SetVariables(Particles.VisibilityTag=&SpawnGroup()) {isSetParameters=true}
   &CameraOffset(Camera Offset Amount=&MaskFloatBySpawnGroup(Masked Float=0, Passthrough Float=10))
  ]
  particleUpdateScript [
   &ParticleState()
   &ScaleSpriteSize(
    Scale Sprite Size Mode=::NewEnumerator0
    Uniform Scale Factor=&FloatFromCurve(FloatCurve=DataInterface() {propertyValues {Curve {keys [
         {arriveTangent=12.8113, interpMode=RCIM_Cubic, leaveTangent=12.8113, tangentMode=RCTM_User, time=0, value=0}
         {arriveTangent=3.3458, interpMode=RCIM_Cubic, leaveTangent=3.3458, tangentMode=RCTM_User, time=0.1, value=1}
         {arriveTangent=-0.0026, tangentMode=RCTM_User, time=1, value=3}
        ]}}})
   ) {module=ScaleSpriteSize002}
   &ScaleColor(
    Scale Alpha=&FloatFromCurve(FloatCurve=DataInterface() {propertyValues {Curve {
        defaultValue=0
        keys [
         {arriveTangent=0.3137, interpMode=RCIM_Cubic, leaveTangent=0.3137, tangentMode=RCTM_User, time=0, value=1}
         {arriveTangent=-2.0958, interpMode=RCIM_Cubic, leaveTangent=-2.0958, tangentMode=RCTM_User, time=1, value=0}
        ]
       }}})
    Scale RGB=&VectorFromCurve(
     VectorCurve=DataInterface() {propertyValues {
       XCurve {keys [{interpMode=RCIM_Cubic, tangentMode=RCTM_User, time=0, value=1}, {arriveTangent=-2.0498, tangentMode=RCTM_User, time=1, value=0}]}
       YCurve {keys [{interpMode=RCIM_Cubic, tangentMode=RCTM_User, time=0, value=1}, {arriveTangent=-2.0498, tangentMode=RCTM_User, time=1, value=0}]}
       ZCurve {keys [{interpMode=RCIM_Cubic, tangentMode=RCTM_User, time=0, value=1}, {arriveTangent=-2.0498, tangentMode=RCTM_User, time=1, value=0}]}
      }}
     Scale Curve=&MakeVectorFromLinearColorRGB(Color=$System.TOD)
    )
    ScaleA=true
    ScaleRGB=true
   )
   /SetVariables(Particles.Glow Color=&Multiply_LinearColorByFloat(
     Linear Color=$Particles.Color
     Float=&FloatFromCurve(FloatCurve=DataInterface() {propertyValues {Curve {
         defaultValue=1
         keys [
          {arriveTangent=23.2094, interpMode=RCIM_Cubic, leaveTangent=23.2094, tangentMode=RCTM_User, time=0, value=0.2}
          {interpMode=RCIM_Cubic, tangentMode=RCTM_User, time=0.1, value=1}
          {arriveTangent=-2.376, interpMode=RCIM_Cubic, leaveTangent=-2.376, tangentMode=RCTM_User, time=1, value=0}
         ]
        }}})
    )) {isSetParameters=true}
   &FN_Sprite_Attributes(
    Color=&MultiplyLinearColor(
     Linear Color=$Particles.Initial.Color
     Scale Factor=&MakeVector4FromVectorAndFloat(
      "Vector (RGB)"=<1 1 1>
      "Float (Alpha)"=&FloatFromCurve(FloatCurve=DataInterface() {propertyValues {Curve {
          defaultValue=1
          keys [
           {arriveTangent=-0.3117, interpMode=RCIM_Cubic, leaveTangent=-0.3117, tangentMode=RCTM_User, time=0.875, value=1}
           {arriveTangent=-14.9834, interpMode=RCIM_Cubic, leaveTangent=-14.9834, tangentMode=RCTM_User, time=1, value=0}
          ]
         }}})
     )
    )
    Create Color Attribute=true
    Create SpriteSize Attribute=true
    SpriteSize=&Multiply_Vector2D_ByFloat(
     Float=&FloatFromCurve(
      FloatCurve=DataInterface() {propertyValues {Curve {
         defaultValue=1
         keys [
          {arriveTangent=6.0858, interpMode=RCIM_Cubic, leaveTangent=6.0858, tangentMode=RCTM_User, time=0, value=0}
          {arriveTangent=2.4343, interpMode=RCIM_Cubic, leaveTangent=2.4343, tangentMode=RCTM_User, time=0.2, value=1}
          {interpMode=RCIM_Cubic, tangentMode=RCTM_User, time=0.8, value=1.6}
         ]
        }}}
      Scale Curve=1.1
     )
     Vector2D=$Particles.Initial.SpriteSize
    )
    Write to Module Namespace=true
   )
   &FN_Light_Attributes(
    Color=<rgba 2.4 3 1.2 1>
    Create Color Attribute=true
    Scale Color=&MaskFloat(
     Mask Source=0
     Mask Parameter=?
     Mask=0
     Masked Float=&FloatFromCurve(
      FloatCurve=DataInterface() {propertyValues {Curve {
         defaultValue=1
         keys [
          {arriveTangent=3.4166, interpMode=RCIM_Cubic, leaveTangent=3.4166, tangentMode=RCTM_User, time=0, value=0.6}
          {arriveTangent=-0.0251, interpMode=RCIM_Cubic, leaveTangent=-0.0251, tangentMode=RCTM_User, time=1, value=0}
         ]
        }}}
      Scale Curve=2
     )
     Passthrough Float=&FloatFromCurve(FloatCurve=DataInterface() {propertyValues {Curve {
         defaultValue=0
         keys [{interpMode=RCIM_Cubic, tangentMode=RCTM_User, time=0, value=0}, {interpMode=RCIM_Cubic, tangentMode=RCTM_User, time=1, value=1}]
        }}})
    )
   )
  ]
  rendererClasses [&NiagaraLightRendererProperties]
  renderers [
   {rendererClass=&NiagaraLightRendererProperties, rendererIndex=0}
   {rendererClass=&NiagaraLightRendererProperties, rendererIndex=1}
  ]
 }
 Light {
  bEnabled=false
  emitterUpdateScript [
   &EmitterState()
   &SpawnBurst_Instantaneous(Use Spawn Probability=false, UseLoopCountLimit=false)
  ]
  particleSpawnScript [&InitializeParticle(Color Mode=::NewEnumerator0, Lifetime=1.5, UsePositionOffset=false, Write Material Random=true)]
  particleUpdateScript [
   &ParticleState()
   &OffsetPosition(Position Offset=$System.IntroPositionOffset)
   &FN_Light_Attributes(
    Color=<rgba 3 2.07 1.2 1>
    Create Color Attribute=true
    Scale Color=&FloatFromCurve(
     FloatCurve=DataInterface() {propertyValues {Curve {
        defaultValue=1
        keys [
         <at 0.5 1 RCIM_Cubic>
         {arriveTangent=0.257, interpMode=RCIM_Cubic, leaveTangent=0.257, tangentMode=RCTM_User, time=0.5587, value=0.7442}
         {arriveTangent=4.2538, interpMode=RCIM_Cubic, leaveTangent=4.2538, time=0.6127, value=1.4144}
         {arriveTangent=3.4166, interpMode=RCIM_Cubic, leaveTangent=3.4166, tangentMode=RCTM_User, time=0.8539, value=2}
         {arriveTangent=-0.0251, interpMode=RCIM_Cubic, leaveTangent=-0.0251, tangentMode=RCTM_User, time=1, value=0}
        ]
       }}}
     Scale Curve=2
    )
   )
  ]
  rendererClasses [&NiagaraLightRendererProperties]
  renderers [{rendererClass=&NiagaraLightRendererProperties, rendererIndex=0}]
 }
 Lightning {
  bEnabled=false
  emitterUpdateScript [
   &EmitterState(Life Cycle Mode=::NewEnumerator1, Loop Behavior=::NewEnumerator1, Loop Duration=1.5, UseLoopDelay=false)
   &SpawnBurst_Instantaneous(Spawn Count=7, Spawn Time=1, Use Spawn Probability=false, UseLoopCountLimit=false)
   &SpawnBurst_Instantaneous(Spawn Count=3, Spawn Group=1, Spawn Time=0.5, Use Spawn Probability=false, UseLoopCountLimit=false) {module=SpawnBurst_Instantaneous001}
   &SpawnRate(
    SpawnGroup=2
    SpawnRate=&FloatFromCurve(
     FloatCurve=DataInterface() {propertyValues {Curve {
        defaultValue=1
        keys [
         <at 0 0 RCIM_Cubic>
         <at 0.1196 0.6932 RCIM_Cubic>
         {arriveTangent=1.1125, arriveTangentWeight=0.0646, interpMode=RCIM_Cubic, leaveTangent=1.1125, tangentMode=RCTM_User, time=0.3135, value=0}
         {arriveTangent=1.2369, interpMode=RCIM_Cubic, leaveTangent=1.2369, time=0.4343, value=3.4471}
         {
          arriveTangent=0.1357
          arriveTangentWeight=0.086
          interpMode=RCIM_Cubic
          leaveTangent=0.1357
          tangentMode=RCTM_User
          time=0.6922
          value=0.4684
         }
         {arriveTangent=-1.5218, interpMode=RCIM_Cubic, leaveTangent=-1.5218, time=0.8588, value=1}
         <at 1 0 RCIM_Cubic>
        ]
       }}}
     CurveIndex=$Emitter.NormalizedLoopAge
    )
    Use Spawn Probability=false
   )
  ]
  particleSpawnScript [
   /SetVariables(
    Particles.LightningIndex:NiagaraInt32=&RandomRangeInt(Minimum=0, Maximum=3)
    Particles.LightningLength:NiagaraFloat=&DistributionCurveFloat(
     Minimum:NiagaraFloat=0.3
     Maximum:NiagaraFloat=0.6
     Distribution Curve:NiagaraDataInterfaceCurve=DataInterface() {propertyValues {Curve {defaultValue=0, keys [<at 0.3 0.2 RCIM_Constant>, <at 0.6 0.3 RCIM_Constant>, {tangentMode=RCTM_User, time=0.85, value=1}]}}}
     Curve Parameter Type:NiagaraInt32=0
     "Curve Parameter Type:":NiagaraInt32=0
     Curve Parameter:NiagaraFloat=?
     Evaluation Type:ENiagaraRandomnessEvaluation=::NewEnumerator0
     Recalculate Random Each Loop:NiagaraBool=true
     Randomness Mode:ENiagaraRandomnessMode=::NewEnumerator0
     Random Seed:NiagaraInt32=0
     Fixed Random Seed:NiagaraBool=false
    )
    Particles.MirrorLightning:NiagaraBool=&RandomBool()
    Particles.Random=&RandomRangeFloat(Minimum=0, Maximum=1)
    Particles.VisibilityTag=&SpawnGroup()
   ) {isSetParameters=true}
   &PartitionParticles(Number Of Partitions=3, Partition Mode:ENiagara_PartitionMode_v2=::NewEnumerator5)
   &InitializeParticle(
    AdjustAlpha=false
    AdjustHue=false
    AdjustSaturation=true
    AdjustValue=true
    Color=<rgba 2 0.1311 0 1>
    Color Mode=::NewEnumerator3
    Lifetime=&RandomRangeFloat(Minimum=0.2, Maximum=0.3)
    Position Offset=$System.BurstPosition
    Saturation Range:Vector2f=<0.95 1>
    UsePositionOffset=true
    Value Range:Vector2f=<1 2>
    Write Material Random=true
   )
   &ShapeLocation(MaskSpawnGroup=false, Override Seed=false, Sphere Radius=1, UseOwnerScale=true)
   &InitialMeshOrientation(
    Mesh Orientation Mode=::NewEnumerator2
    Orientation Vector=&FN_FrontEndVector3Modifier(Gameplay=<-1 0 0>, Front End=&RandomVector(Vector Scale=1))
    Rotation=&MakeVector(
     X=&Add_Float(
      A=&FloatFromCurve(
       FloatCurve=DataInterface() {propertyValues {Curve {keys [<at 0 1>,<at 1 0>]}}}
       CurveIndex=&ReturnNormalizedExecIndex()
      )
      B=&RandomRangeFloat(Minimum=-0.1, Maximum=0.1)
     )
     Y=&RandomRangeFloat(Minimum=-0.3, Maximum=0.3)
     Z=0
    )
    Use Rotation Vector=true
   )
   &CameraOffset(Camera Offset Amount=5)
  ]
  particleUpdateScript [
   &ParticleState()
   &FN_Mesh_Attributes(
    Create Scale Attribute=true
    Scale=&Multiply_Vector(
     A=&Multiply_VectorByFloat(
      Vector=&MakeCustomVectorFromBool(Bool=$Particles.Initial.MirrorLightning, True Vector=<1 -1 -1>, False Vector=<1 1 1>)
      Float=$Particles.Initial.LightningLength
     )
     B=&Multiply_VectorByFloat(
      Vector=&MakeVector(
       X=&FloatFromCurve(FloatCurve=DataInterface() {propertyValues {Curve {
           defaultValue=1
           keys [
            {arriveTangent=0.8401, interpMode=RCIM_Cubic, leaveTangent=0.8401, tangentMode=RCTM_User, time=0, value=0.8}
            {interpMode=RCIM_Cubic, tangentMode=RCTM_User, time=1, value=1}
           ]
          }}})
       Y=1
       Z=1
      )
      Float=$System.FE_Scale
     )
    )
   )
   /SetVariables(
    Particles.DynamicParametersAlt:Vector4f=&MakeVector4(
     X=&Multiply_Float(A=$Particles.Initial.LightningLength, B=&RandomRangeFloat(Minimum=0.5, Maximum=1.5))
     Y=&UniformAOrBFloat(A=0.5, B=1.5, Evaluation Type=::NewEnumerator0, Recalculate Random Each Loop=true, Override Seed=false)
     Z=&FloatFromCurve(FloatCurve=DataInterface() {propertyValues {Curve {
         defaultValue=1
         keys [
          {arriveTangent=-2.478, interpMode=RCIM_Cubic, leaveTangent=-2.478, tangentMode=RCTM_User, time=0, value=1}
          {interpMode=RCIM_Cubic, tangentMode=RCTM_User, time=1, value=0}
         ]
        }}})
     W=0
    )
    Particles.DynamicParametersAlt2:Vector4f=&MakeVector4(X=1, Y=1, Z=0, W=0)
   ) {isSetParameters=true}
   &FN_Mesh_Attributes(
    Color=&MultiplyLinearColor(
     Linear Color=&RandomRangeLinearColor(Minimum=<rgba 5 1 5 1>, Maximum=<rgba 5 1 3.6667 1>)
     Scale Factor=&MakeVector4FromVectorAndFloat(
      "Vector (RGB)"=&VectorFromFloat(Value=&RandomRangeFloat(Minimum=0, Maximum=2, Evaluation Type=::NewEnumerator1))
      "Float (Alpha)"=&FloatFromCurve(FloatCurve=DataInterface() {propertyValues {Curve {keys [<at 0.7 1>,<at 1 0>]}}})
     )
    )
    Create Color Attribute=true
    Create Orientation Attribute=true
    Create Scale Attribute=true
    Orientation:Quat4f=&MakeQuaternion(
     Quaternion From=1
     Axis=&RandomRangeVector(Minimum=<-0.3 -0.3 -1>, Maximum=<0.3 0.3 1>)
     Angle=&RandomRangeFloat(Minimum=0, Maximum=360)
     Yaw=&RandomRangeFloat(Minimum=-0.1, Maximum=0.1)
     Pitch=&RandomRangeFloat(Minimum=0.05, Maximum=0.45)
     Roll=&RandomRangeFloat(Minimum=-1, Maximum=1)
     From=?
     To=?
     Angle Type=::NewEnumerator1
     Coordinate Space=::Local
     Rotation Matrix=Linked() {linkedVariable {name=Engine.Owner.SystemLocalToWorldNoScale, type=NiagaraInt32, defaultValue=~, description=~}}
     Normalize Vectors=false
    )
    Scale=&Multiply_VectorByFloat(
     Vector=&VectorFromFloat(Value=&Multiply_Float(A=0.3, B=0.6))
     Float=&FloatFromCurve(
      FloatCurve=DataInterface() {propertyValues {Curve {keys [
          {arriveTangent=3.2151, interpMode=RCIM_Cubic, leaveTangent=3.2151, tangentMode=RCTM_User, time=0, value=0.6}
          {arriveTangent=0.4, interpMode=RCIM_Cubic, leaveTangent=0.4, time=0.5973, value=0.9154}
          {arriveTangent=0.0224, interpMode=RCIM_Cubic, leaveTangent=0.0224, tangentMode=RCTM_User, time=1, value=1}
         ]}}}
      Scale Curve=$System.FE_Scale
     )
    )
    Write to Module Namespace=true
   ) {module=FN_Mesh_Attributes002}
   /SetVariables(
    Particles.DynamicParametersAlt1:Vector4f=&MakeVector4(X=0, Y=0, Z=0, W=13)
    Particles.DynamicParametersAlt3:Vector4f=&MakeVector4(
     X=&FloatFromCurve(FloatCurve=DataInterface() {propertyValues {Curve {keys [<at 0 0.5>,<at 1 1>]}}})
     Y=3
     Z=&FloatFromCurve(FloatCurve=DataInterface() {propertyValues {Curve {keys [{interpMode=RCIM_Cubic, tangentMode=RCTM_Break, time=0.5, value=-0.5}, <at 1 0.1>]}}})
     W=3
    )
    Particles.DynamicParametersAlt4:Vector4f=&MakeVector4(
     X=&Multiply_Float(
      A=&FloatFromCurve(FloatCurve=DataInterface() {propertyValues {Curve {keys [{interpMode=RCIM_Cubic, leaveTangent=8.1085, tangentMode=RCTM_Break, time=0.8568, value=1}, {tangentMode=RCTM_User, time=1, value=0}]}}})
      B=&Divide_Float(
       A=&MakeFloatFromVector(Vector=$Particles.FN_Mesh_Attributes002.Scale, Channel=::NewEnumerator0)
       B=&RandomRangeFloat(Minimum=50, Maximum=100)
      )
     )
     Y=&FloatFromCurve(
      FloatCurve=DataInterface() {propertyValues {Curve {keys [
          {interpMode=RCIM_Cubic, tangentMode=RCTM_Break, time=0.7, value=0}
          {arriveTangent=-0.0859, interpMode=RCIM_Cubic, tangentMode=RCTM_Break, time=1, value=-1}
         ]}}}
      Scale Curve=50
     )
     Z=&Waveform(
      Global Input Position=&RandomRangeFloat(Minimum=0, Maximum=15)
      Global Phase=$Particles.Age
      "Waveform [1]"=::NewEnumerator12
      "[1] Phase Scale"=2
      "[1] Frequency"=1
      "[1] Amplitude Min/Max"=<0 1>
     )
     W=25
    )
    Particles.DynamicParametersAlt5:Vector4f=&MakeVector4(
     X=&FloatFromCurve(
      FloatCurve=DataInterface() {propertyValues {Curve {
         defaultValue=1
         keys [
          {arriveTangent=2.157, interpMode=RCIM_Cubic, leaveTangent=2.157, tangentMode=RCTM_User, time=0, value=0}
          {interpMode=RCIM_Cubic, tangentMode=RCTM_User, time=1, value=1}
         ]
        }}}
      Scale Curve=10
     )
     Y=0.1
     Z=&FloatFromCurve(FloatCurve=DataInterface() {propertyValues {Curve {keys [
          {interpMode=RCIM_Cubic, leaveTangent=1.1034, tangentMode=RCTM_Break, time=0, value=-1}
          {arriveTangent=2.2399, tangentMode=RCTM_User, time=0.5, value=0.125}
         ]}}})
     W=&FloatFromCurve(FloatCurve=DataInterface() {propertyValues {Curve {keys [
          {interpMode=RCIM_Cubic, leaveTangent=1.8089, tangentMode=RCTM_Break, time=0.3, value=-0.5}
          {arriveTangent=2.0757, interpMode=RCIM_Cubic, tangentMode=RCTM_Break, time=0.9, value=1}
         ]}}})
    )
   ) {isSetParameters=true}
   &ScaleColor(
    Scale Alpha=&Multiply_Float(
     A=10
     B=&FloatFromCurve(FloatCurve=DataInterface() {propertyValues {Curve {keys [<at 0 0>,<at 0.1 0>,<at 0.2 0.8>,<at 1 0>]}}})
    )
    Scale RGB=&Multiply_VectorByFloat(
     Vector=&Multiply_VectorByFloat(Vector=<1 1 1>, Float=&FN_FrontEndFloatModifier(Gameplay=0.5, Front End=0.5))
     Float=&Multiply_Float(
      A=&RandomRangeFloat(Minimum=0.35, Maximum=1, Evaluation Type=::NewEnumerator1)
      B=&Multiply_Float(A=&FNTimeOfDayScalar(Night Scale=0.2, Morning Scale=1, Evening Scale=1), B=10)
     )
    )
    ScaleA=true
    ScaleRGB=true
   )
   &DynamicMaterialParameters(
    Index 0 Param 1=1
    Index 0 Param 2=0.7
    Index 0 Param 3=&FloatFromCurve(FloatCurve=DataInterface() {propertyValues {Curve {keys [
         {arriveTangent=4.0072, interpMode=RCIM_Cubic, leaveTangent=4.0072, tangentMode=RCTM_User, time=0, value=0}
         {arriveTangent=2.7083, interpMode=RCIM_Cubic, leaveTangent=2.7083, time=0.2543, value=0.6225}
         {arriveTangent=1.8473, interpMode=RCIM_Cubic, leaveTangent=1.8473, time=0.6, value=1.625}
         {arriveTangent=1.6993, interpMode=RCIM_Cubic, leaveTangent=1.6993, tangentMode=RCTM_User, time=1, value=2}
        ]}}})
    Index 0 Param 4=13
    Index 1 Param 1=&FN_FrontEndFloatModifier(Gameplay=1.2, Front End=0.5)
    Index 1 Param 2=$Particles.Initial.Random
    Index 1 Param 3=&FloatFromCurve(FloatCurve=DataInterface() {propertyValues {Curve {keys [<at 0 0.1>,<at 1 1.5>]}}})
    Index 1 Param 4=1
    Index 2 Param 1=&FloatFromCurve(FloatCurve=DataInterface() {propertyValues {Curve {keys [<at 0 1>,<at 1 0.3>]}}})
    Module.Index0Float0Write=true
    Module.Index0Float1Write=true
    Module.Index0Float2Write=true
    Module.Index0Float3Write=true
    Module.Index1Float0Write=true
    Module.Index1Float1Write=true
    Module.Index1Float2Write=true
    Module.Index1Float3Write=true
    Module.Index2Float0Write=true
    Module.Index2Float1Write=true
    Module.Index2Float2Write=true
    Module.Index2Float3Write=true
    Write Parameter Index 1=::NewEnumerator1
    Write Parameter Index 2=::NewEnumerator1
   )
   /SetVariables(Particles.Lightningglowcolor:LinearColor=&MakeLinearColorFromVectorAndFloat(
     "Vector (RGB)"=&MakeVectorFromLinearColorRGB(Color=&MultiplyLinearColor(
       Linear Color=<rgba 5 2.6797 0.3594 1>
       Scale Factor=&MakeVector4_FromFloat(Float=&RandomRangeFloat(Minimum=0.6, Maximum=0.8, Evaluation Type=::NewEnumerator1))
      ))
     "Float (Alpha)"=&Multiply_Float(
      A=&FloatFromCurve(FloatCurve=DataInterface() {propertyValues {Curve {keys [<at 0 1 RCIM_Cubic>, {interpMode=RCIM_Cubic, leaveTangent=-3.6307, time=0.9, value=0}]}}})
      B=1
     )
    )) {isSetParameters=true}
   /SetVariables(Particles.Glow Dy02:Vector4f=&MakeVector4(
     X=&FloatFromCurve(FloatCurve=DataInterface() {propertyValues {Curve {keys [{tangentMode=RCTM_User, time=0, value=0.2}, {tangentMode=RCTM_User, time=0.5, value=1}]}}})
     Y=0.7
     Z=0
     W=13
    )) {isSetParameters=true}
   /SetVariables(Particles.Glow Dy03:Vector4f=&MakeVector4(
     X=&FloatFromCurve(
      FloatCurve=DataInterface() {propertyValues {Curve {keys [<at 0 1>,<at 1 0.5>]}}}
      Scale Curve=&FN_FrontEndFloatModifier(Gameplay=1.5, Front End=1)
     )
     Y=$Particles.Initial.Random
     Z=1
     W=1
    )) {isSetParameters=true}
  ]
  rendererClasses [&NiagaraMeshRendererProperties]
  renderers [
   {rendererClass=&NiagaraMeshRendererProperties, rendererIndex=0}
   {rendererClass=&NiagaraMeshRendererProperties, rendererIndex=1}
   {rendererClass=&NiagaraMeshRendererProperties, rendererIndex=2}
  ]
 }
 Main_Heart {
  emitterUpdateScript [
   &EmitterState(Life Cycle Mode=::NewEnumerator1, Loop Behavior=::NewEnumerator1, Loop Duration=1.5333, UseLoopDelay=false)
   &SpawnBurst_Instantaneous(Use Spawn Probability=false, UseLoopCountLimit=false)
   /Niagara/Modules/Spawn/Location/SpawnParticlesInGrid(Spawn Group=1, X Count=1, Y Count=11, Z Count=9)
  ]
  particleSpawnScript [
   /SetVariables(Particles.VisibilityTag=$Emitter.SpawnGroup) {isSetParameters=true}
   &InitializeParticle(
    Color=<rgba 1 0.5 0 1>
    Lifetime=1.5
    Mesh Scale Mode=::NewEnumerator3
    Mesh Uniform Scale=0.21
    Position Offset=$System.BurstPosition
    Sprite Rotation Angle=2
    Sprite Rotation Mode=::NewEnumerator0
    Sprite Size=<42.0171 39.864>
    Sprite Size Mode=::NewEnumerator4
    UsePositionOffset=true
    Write Material Random=true
   )
   /Niagara/Modules/Spawn/Location/V2/GridLocation(
    Module.Jitter Inside Grid Cells=false
    Module.Override Random Seed=false
    Offset:Vector3f=<0.5 0.5 0.5>
    XYZ Dimensions:Vector3f=&Multiply_VectorByFloat(Vector=<2 2 2>, Float=1)
   )
   &InitialMeshOrientation(Mesh Orientation Mode=::NewEnumerator2, Rotation=<0.01 0 0.07>, Use Rotation Vector=true)
   /SetVariables(
    Particles.GlowCameraOffset=-4.5082
    Particles.GridCubeColor:LinearColor=&Multiply_LinearColorByFloat(Linear Color=$Particles.Color, Float=&RandomRangeFloat(Minimum=1, Maximum=1.1))
    Particles.HeartPosition:NiagaraPosition=<0 0 0>
   ) {isSetParameters=true}
   &AddVelocity(Module.Use Falloff Distance=false, Origin Offset=<0 0 -5>, Override Seed=false, Velocity Mode=::NewEnumerator1, Velocity Speed=15)
   &ApplyInitialForces()
   &SubUVAnimation(
    Mesh Renderer=DataInterface() {propertyValues {MeshRenderer=&self:Main_Heart_0.NiagaraMeshRendererProperties_1}}
    Output Alternate Renderer Binding=true
    Random Start Frame=true
    Renderer Type:ENiagaraMeshOrSprite=::NewEnumerator1
    SubUV Lookup Index Scale=0
    UseEndFrame=false
    UseStartFrame=false
   ) {enabled=false}
  ]
  particleUpdateScript [
   &ParticleState()
   &ScaleSpriteSize(
    Uniform Curve Scale=&Multiply_Float(A=0.7, B=1)
    Uniform Curve Sprite Scale=DataInterface() {propertyValues {Curve {keys [
        {tangentMode=RCTM_User, time=0, value=0}
        {
         arriveTangent=0.3363
         arriveTangentWeight=0.0333
         interpMode=RCIM_Cubic
         leaveTangent=0.3363
         leaveTangentWeight=0.3
         tangentMode=RCTM_User
         time=0.3335
         value=0
        }
        {arriveTangent=0.112, interpMode=RCIM_Cubic, tangentMode=RCTM_User, time=0.48, value=1}
        {arriveTangent=1.1752, interpMode=RCIM_Cubic, leaveTangent=1.1752, tangentMode=RCTM_User, time=0.666, value=0.8}
        {arriveTangent=0.5988, interpMode=RCIM_Cubic, leaveTangent=0.5988, time=0.8, value=1.4}
        {tangentMode=RCTM_User, time=1, value=1}
       ]}}}
   )
   &ScaleMeshSize(Scale Factor=&Multiply_VectorByFloat(
     Vector=<1 1 1>
     Float=&FloatFromCurve(
      FloatCurve=DataInterface() {propertyValues {Curve {
         defaultValue=1
         keys [
          {arriveTangent=21.6274, interpMode=RCIM_Cubic, leaveTangent=21.6274, tangentMode=RCTM_User, time=0, value=0}
          {arriveTangent=0.08, arriveTangentWeight=0.0504, interpMode=RCIM_Cubic, leaveTangent=0.08, tangentMode=RCTM_User, time=0.1512, value=1}
          {arriveTangentWeight=0.0385, interpMode=RCIM_Cubic, tangentMode=RCTM_User, time=0.2666, value=1}
          <at 0.4 0 RCIM_Constant>
          {arriveTangent=3.4736, interpMode=RCIM_Cubic, leaveTangent=3.4736, time=0.75, value=0}
          {
           arriveTangentWeight=0.05
           interpMode=RCIM_Cubic
           leaveTangentWeight=0.0449
           tangentMode=RCTM_Break
           tangentWeightMode=RCTWM_WeightedBoth
           time=0.9
           value=1.7368
          }
          <at 1 0 RCIM_Constant>
         ]
        }}}
      Scale Curve=$System.FE_Scale
     )
    ))
   &ScaleColor(
    Scale RGB=&Multiply_VectorByFloat(
     Vector=<1 1 1>
     Float=&Lerp_Float(
      A=&FloatFromCurve(FloatCurve=DataInterface() {propertyValues {Curve {keys [
           {interpMode=RCIM_Cubic, leaveTangentWeight=0.0167, tangentMode=RCTM_User, time=0, value=1}
           {arriveTangentWeight=0.085, interpMode=RCIM_Cubic, leaveTangentWeight=0.085, tangentMode=RCTM_User, time=0.05, value=2}
           {arriveTangentWeight=0.0167, interpMode=RCIM_Cubic, leaveTangentWeight=0.0167, tangentMode=RCTM_User, time=0.1, value=0.75}
           {arriveTangentWeight=0.0175, interpMode=RCIM_Cubic, leaveTangentWeight=0.0324, tangentMode=RCTM_User, time=0.15, value=2}
           {arriveTangentWeight=0.1101, interpMode=RCIM_Cubic, leaveTangentWeight=0.059, tangentMode=RCTM_User, time=0.2427, value=0.75}
           {arriveTangentWeight=0.0168, interpMode=RCIM_Cubic, leaveTangentWeight=0.0703, tangentMode=RCTM_User, time=0.2924, value=1.5132}
           {arriveTangent=0.5888, interpMode=RCIM_Cubic, leaveTangent=0.5888, tangentMode=RCTM_User, time=0.5, value=0.3}
           <at 0.6877 1 RCIM_Cubic>
           {arriveTangent=-0.515, interpMode=RCIM_Cubic, leaveTangent=-0.515, tangentMode=RCTM_User, time=0.7477, value=0.3}
           {arriveTangent=1.0388, interpMode=RCIM_Cubic, leaveTangent=1.0388, tangentMode=RCTM_User, time=0.7952, value=3}
           {
            arriveTangent=1.3795
            arriveTangentWeight=0.1142
            interpMode=RCIM_Cubic
            leaveTangent=3.0389
            leaveTangentWeight=0.5165
            tangentMode=RCTM_Break
            tangentWeightMode=RCTWM_WeightedBoth
            time=0.8466
            value=1
           }
           {arriveTangent=63.7325, tangentMode=RCTM_User, time=0.9985, value=5}
          ]}}})
      B=&FloatFromCurve(FloatCurve=DataInterface() {propertyValues {Curve {keys [<at 0 2>,<at 0.0912 2>,<at 0.2854 2>,<at 0.9021 2>,<at 1 1>]}}})
      Alpha=&MakeFloatFromInt(Int=&SpawnGroup())
     )
    )
    ScaleA=true
    ScaleRGB=true
   )
   /SetVariables(
    Particles.GridCubePosition:NiagaraPosition=&SubtractVectorFromPosition(Position=$Particles.Position, Vector=$System.IntroPositionOffset)
    Particles.HeartColor:LinearColor=&MakeCustomLinearColorFromBool(
     Bool=true
     True Color=&ColorFromCurve(ColorCurve=DataInterface() {propertyValues {
        AlphaCurve {keys [
          {interpMode=RCIM_Cubic, tangentMode=RCTM_User, time=0.1954, value=0}
          {arriveTangent=6.9335, tangentMode=RCTM_User, time=0.5325, value=1}
          <at 0.5837 1 RCIM_Cubic>
          <at 0.9558 0 RCIM_Cubic>
         ]}
        BlueCurve {keys [
          {interpMode=RCIM_Cubic, leaveTangentWeight=0.3333, tangentMode=RCTM_User, time=0, value=0.005}
          <at 0.4655 0>
          <at 0.5343 0>
          <at 0.5851 0>
          <at 0.6848 0>
          <at 0.9492 0 RCIM_Cubic>
         ]}
        GreenCurve {keys [
          {interpMode=RCIM_Cubic, leaveTangentWeight=0.3333, tangentMode=RCTM_User, time=0, value=0.0525}
          <at 0.4655 0.695>
          <at 0.5343 0.695>
          <at 0.5851 0.5>
          <at 0.6848 0.35>
          <at 0.9492 0.2 RCIM_Cubic>
         ]}
        RedCurve {keys [
          {interpMode=RCIM_Cubic, leaveTangentWeight=0.3333, tangentMode=RCTM_User, time=0, value=0.1}
          <at 0.4655 2>
          <at 0.5343 2>
          <at 0.5851 1>
          <at 0.6848 0.7>
          <at 0.9492 0.4 RCIM_Cubic>
         ]}
       }})
     False Color=&ColorFromCurve(ColorCurve=DataInterface() {propertyValues {
        AlphaCurve {keys [
          {interpMode=RCIM_Cubic, tangentMode=RCTM_User, time=0.1954, value=0}
          {arriveTangent=6.9335, tangentMode=RCTM_User, time=0.5325, value=1}
          <at 0.5837 1 RCIM_Cubic>
          <at 0.9558 0 RCIM_Cubic>
         ]}
        BlueCurve {keys [
          {interpMode=RCIM_Cubic, leaveTangentWeight=0.3333, tangentMode=RCTM_User, time=0, value=0.1}
          <at 0.4655 0.5>
          <at 0.5343 0.5>
          <at 0.5851 1>
          <at 0.6848 0.7>
          <at 0.9492 0.4 RCIM_Cubic>
         ]}
        GreenCurve {keys [
          {interpMode=RCIM_Cubic, leaveTangentWeight=0.3333, tangentMode=RCTM_User, time=0, value=0.1}
          <at 0.4655 0.5>
          <at 0.5343 0.5>
          <at 0.5851 1>
          <at 0.6848 0.7>
          <at 0.9492 0.4 RCIM_Cubic>
         ]}
        RedCurve {keys [
          {interpMode=RCIM_Cubic, leaveTangentWeight=0.3333, tangentMode=RCTM_User, time=0, value=0.1}
          <at 0.4655 0.5>
          <at 0.5343 0.5>
          <at 0.5851 1>
          <at 0.6848 0.7>
          <at 0.9492 0.4 RCIM_Cubic>
         ]}
       }})
    )
    Particles.HeartMeshScale:Vector3f=&Multiply_VectorByFloat(
     Vector=$Particles.Initial.Scale
     Float=&FloatFromCurve(FloatCurve=DataInterface() {propertyValues {Curve {keys [
          <at 0 1>
          {arriveTangentWeight=0.066, interpMode=RCIM_Cubic, tangentMode=RCTM_User, time=0.198, value=1}
          {
           arriveTangent=0.082
           arriveTangentWeight=0.0573
           interpMode=RCIM_Cubic
           leaveTangent=0.0753
           leaveTangentWeight=0.1074
           tangentMode=RCTM_Break
           tangentWeightMode=RCTWM_WeightedBoth
           time=0.329
           value=1.3252
          }
          {
           arriveTangentWeight=0.3342
           interpMode=RCIM_Cubic
           leaveTangentWeight=0.0447
           tangentMode=RCTM_Break
           tangentWeightMode=RCTWM_WeightedBoth
           time=0.666
           value=1.0068
          }
          {
           arriveTangent=2.2685
           arriveTangentWeight=0.0447
           interpMode=RCIM_Cubic
           leaveTangent=2.2685
           tangentMode=RCTM_User
           time=0.7977
           value=2.0226
          }
          {
           arriveTangent=2.3272
           arriveTangentWeight=0.0167
           interpMode=RCIM_Cubic
           leaveTangent=2.3272
           tangentMode=RCTM_User
           time=0.8883
           value=2.4506
          }
          {arriveTangent=-5.0397, arriveTangentWeight=0.05, interpMode=RCIM_Cubic, leaveTangent=-5.0397, tangentMode=RCTM_User, time=1, value=0.7}
         ]}}})
    )
    Particles.HeartPosition:Vector3f=$System.IntroPositionOffset
    Particles.MeshOrientation=&MakeQuaternion(
     Quaternion From=0
     Axis=<0 0 1>
     Angle=&FloatFromCurve(
      FloatCurve=DataInterface() {propertyValues {Curve {
         defaultValue=0
         keys [
          {arriveTangent=2.5149, interpMode=RCIM_Cubic, leaveTangent=2.5149, tangentMode=RCTM_User, time=0, value=-0.3}
          {arriveTangent=0.0022, interpMode=RCIM_Cubic, leaveTangent=0.0022, tangentMode=RCTM_User, time=0.3, value=0}
          {arriveTangent=1.789, tangentMode=RCTM_User, time=1.0013, value=0.1863}
         ]
        }}}
      Scale Curve=560
     )
     From=?
     To=?
     Angle Type=::NewEnumerator0
     Rotation Matrix=Linked() {linkedVariable {name=Engine.Owner.SystemLocalToWorldNoScale, type=NiagaraInt32, defaultValue=~, description=~}}
     Normalize Vectors=false
    )
   ) {isSetParameters=true}
   &self:GimbalLock(Gimbal Lock Direction:NiagaraPosition=&ReturnCameraVector(
     Camera Vectors=::NewEnumerator4
     Coordinate Space=::Local
     Origin Position=$Engine.Emitter.SimulationPosition
     CameraQuery=DataInterface() {propertyValues {bRequireCurrentFrameData=false}}
    )) {module=GimbalLock}
   /SetVariables(
    Particles.GlowColor=&ScaleLinearColorByCurve(
     InitialValue=$Particles.Initial.Color
     ScaleCurve=DataInterface() {propertyValues {
       AlphaCurve {
        defaultValue=0
        keys [
         {interpMode=RCIM_Cubic, leaveTangentWeight=0.1058, tangentMode=RCTM_Break, time=0.2523, value=1}
         {arriveTangentWeight=0.1058, interpMode=RCIM_Cubic, leaveTangentWeight=0.0235, tangentMode=RCTM_User, time=0.5696, value=1}
         {arriveTangent=-2.3236, interpMode=RCIM_Cubic, leaveTangent=-2.3236, time=0.64, value=1.5}
         <at 1 0>
        ]
       }
       BlueCurve {
        defaultValue=0
        keys [
         {interpMode=RCIM_Cubic, tangentMode=RCTM_Break, time=0.2523, value=1}
         {arriveTangentWeight=0.1058, interpMode=RCIM_Cubic, tangentMode=RCTM_User, time=0.5696, value=1}
         {arriveTangent=-2.3236, interpMode=RCIM_Cubic, leaveTangent=-2.3236, time=0.64, value=1.5}
         <at 1 0>
        ]
       }
       GreenCurve {
        defaultValue=0
        keys [
         {interpMode=RCIM_Cubic, tangentMode=RCTM_Break, time=0.2523, value=1}
         {arriveTangentWeight=0.1058, interpMode=RCIM_Cubic, tangentMode=RCTM_User, time=0.5696, value=1}
         {arriveTangent=-2.3236, interpMode=RCIM_Cubic, leaveTangent=-2.3236, time=0.64, value=1.5}
         <at 1 0>
        ]
       }
       RedCurve {
        defaultValue=0
        keys [
         {interpMode=RCIM_Cubic, tangentMode=RCTM_Break, time=0.2523, value=1}
         {arriveTangentWeight=0.1058, interpMode=RCIM_Cubic, tangentMode=RCTM_User, time=0.5696, value=1}
         {arriveTangent=-2.3236, interpMode=RCIM_Cubic, leaveTangent=-2.3236, time=0.64, value=1.5}
         <at 1 0>
        ]
       }
      }}
     CurveIndex=$Particles.NormalizedAge
    )
    Particles.GridCubeScale:Vector3f=/Niagara/DynamicInputs/Math/LerpMultipleVectors(
     Alpha:NiagaraFloat=&Add_Float(
      A=/Niagara/DynamicInputs/Math/Floor_Float(Float:NiagaraFloat=&Multiply_Float(
        A=&Add_Float(
         A=&Add_Float(
          A=&RemapRange(
           Input Value To Remap=&DistanceBetweenPositions(
            A=&Lerp_Vector(A=$System.IntroPositionOffset, B=<0.5 -7.5 1.5>, Alpha=$System.TimelineCurve)
            B=$Particles.GridCubePosition
           )
           Input Min=0
           Input Max=7.5
           Output Min=0
           Output Max=1
          )
          B=&RemapRange(
           Input Value To Remap=&DistanceBetweenPositions(
            A=&Lerp_Vector(A=$System.IntroPositionOffset, B=<0.5 7.5 1.5>, Alpha=$System.TimelineCurve)
            B=$Particles.GridCubePosition
           )
           Input Min=0
           Input Max=7.5
           Output Min=0
           Output Max=1
          )
         )
         B=&RemapRange(
          Input Value To Remap=&DistanceBetweenPositions(A=<1.5 0 -6.4196>, B=$Particles.GridCubePosition)
          Input Min=0
          Input Max=4.1386
          Output Min=0
          Output Max=1
         )
        )
        B=&FloatFromCurve(
         FloatCurve=DataInterface() {propertyValues {Curve {keys [
             {interpMode=RCIM_Cubic, tangentMode=RCTM_User, time=0, value=3.6558}
             <at 0.3161 1 RCIM_Constant>
             {interpMode=RCIM_Cubic, leaveTangent=-8.3776, time=0.3929, value=1}
            ]}}}
         Scale Curve=0.335
        )
       ))
      B=&FloatFromCurve(
       FloatCurve=DataInterface() {propertyValues {Curve {keys [
           {interpMode=RCIM_Cubic, tangentMode=RCTM_User, time=0.3502, value=0}
           <at 0.3968 1>
           {interpMode=RCIM_Cubic, tangentMode=RCTM_User, time=0.7, value=1}
           <at 0.77 -1 RCIM_Cubic>
           <at 1 1>
          ]}}}
       CurveIndex=&Add_Float(A=$Particles.NormalizedAge, B=&RandomRangeFloat(Minimum=-0.1, Maximum=0.1))
      )
     )
     V 1:Vector3f=$Particles.Initial.Scale
     V 2:Vector3f=<0 0 0>
     V 3:Vector3f=?
     V 4:Vector3f=?
     V 5:Vector3f=?
     Vector Count:ENiagara_MultipleLerpCount=::NewEnumerator3
    )
   ) {isSetParameters=true}
   &DynamicMaterialParameters(
    Index 0 Param 1=2
    Index 0 Param 2=1
    Index 0 Param 3=&Waveform("Waveform [1]"=::NewEnumerator0, "[1] Frequency"=0.8, "[1] Amplitude Min/Max"=<1 2>)
    Module.Index0Float0Write=true
    Module.Index0Float1Write=true
    Module.Index0Float2Write=true
    Module.Index0Float3Write=true
   ) {module=DynamicMaterialParameters001}
   &ScaleVelocity(Velocity Scale=&Multiply_VectorByFloat(
     Vector=<1 1 1>
     Float=&FloatFromCurve(FloatCurve=DataInterface() {propertyValues {Curve {keys [
          <at 0.3 0 RCIM_Cubic>
          <at 0.3276 -1.9126 RCIM_Cubic>
          {arriveTangent=5.9317, interpMode=RCIM_Cubic, leaveTangent=5.9317, time=0.5003, value=0}
          {interpMode=RCIM_Cubic, tangentMode=RCTM_User, time=0.65, value=0}
          {arriveTangent=4.2086, interpMode=RCIM_Cubic, leaveTangent=4.2086, tangentMode=RCTM_User, time=0.7101, value=5}
          {
           arriveTangent=-0.5056
           arriveTangentWeight=0.0762
           interpMode=RCIM_Cubic
           leaveTangent=-0.5056
           leaveTangentWeight=0.0762
           tangentMode=RCTM_User
           tangentWeightMode=RCTWM_WeightedBoth
           time=0.85
           value=0
          }
          {arriveTangent=-84.478, tangentMode=RCTM_User, time=1, value=-30}
         ]}}})
    ))
   &SolveForcesAndVelocity(Clamp Velocity=false, Limit Acceleration=false)
   /SetVariables(
    Particles.SoftGlowColor:LinearColor=&ScaleLinearColorByCurve(
     InitialValue=$Particles.Color
     ScaleCurve=DataInterface() {propertyValues {AlphaCurve {keys [<at 0 0.3>]}, BlueCurve {keys [<at 0 1>]}, GreenCurve {keys [<at 0 1>]}, RedCurve {keys [<at 0 1>]}}}
     CurveIndex=$Particles.NormalizedAge
    )
    Particles.SoftGlowSize:Vector2f=&Multiply_Vector2D_ByFloat(
     Float=&FloatFromCurve(FloatCurve=DataInterface() {propertyValues {Curve {keys [
          {interpMode=RCIM_Cubic, leaveTangentWeight=0.3333, tangentMode=RCTM_User, time=0, value=0.1}
          {arriveTangentWeight=0.0167, interpMode=RCIM_Cubic, leaveTangentWeight=0.3167, tangentMode=RCTM_User, time=0.05, value=0.3}
          {arriveTangentWeight=0.0167, interpMode=RCIM_Cubic, leaveTangentWeight=0.3, tangentMode=RCTM_User, time=0.1, value=0.5}
          {arriveTangentWeight=0.0167, interpMode=RCIM_Cubic, leaveTangentWeight=0.2833, tangentMode=RCTM_User, time=0.1509, value=0.6566}
          {arriveTangentWeight=0.0309, interpMode=RCIM_Cubic, leaveTangentWeight=0.2524, tangentMode=RCTM_User, time=0.2427, value=0.4562}
          {arriveTangentWeight=0.0166, interpMode=RCIM_Cubic, leaveTangentWeight=0.2359, tangentMode=RCTM_User, time=0.2924, value=0.9}
          {arriveTangent=-1.9271, interpMode=RCIM_Cubic, leaveTangent=-1.9271, time=0.3806, value=0.9298}
          {
           arriveTangent=0.5888
           arriveTangentWeight=0.0692
           interpMode=RCIM_Cubic
           leaveTangent=0.5888
           leaveTangentWeight=0.1667
           tangentMode=RCTM_User
           time=0.5
           value=0.5
          }
          {arriveTangent=6.0563, interpMode=RCIM_Cubic, leaveTangent=6.0563, time=0.6877, value=1}
          {
           arriveTangent=-0.515
           arriveTangentWeight=0.02
           interpMode=RCIM_Cubic
           leaveTangent=-0.515
           leaveTangentWeight=0.0841
           tangentMode=RCTM_User
           time=0.7477
           value=2
          }
          {
           arriveTangent=1.0388
           arriveTangentWeight=0.0158
           interpMode=RCIM_Cubic
           leaveTangent=1.0388
           leaveTangentWeight=0.0683
           tangentMode=RCTM_User
           time=0.7952
           value=1.1303
          }
          {
           arriveTangent=1.3795
           arriveTangentWeight=0.1142
           interpMode=RCIM_Cubic
           leaveTangent=3.0389
           leaveTangentWeight=0.5165
           tangentMode=RCTM_Break
           tangentWeightMode=RCTWM_WeightedBoth
           time=0.8466
           value=0.5872
          }
          <at 1 0>
         ]}}})
     Vector2D=$Particles.Initial.SpriteSize
    )
   ) {isSetParameters=true}
  ]
  rendererClasses [&NiagaraMeshRendererProperties, &NiagaraSpriteRendererProperties]
  renderers [
   {rendererClass=&NiagaraMeshRendererProperties, rendererIndex=0}
   {rendererClass=&NiagaraSpriteRendererProperties, rendererIndex=1}
   {rendererClass=&NiagaraMeshRendererProperties, rendererIndex=2}
   {rendererClass=&NiagaraSpriteRendererProperties, rendererIndex=3}
   {rendererClass=&NiagaraMeshRendererProperties, rendererIndex=4}
  ]
 }
 Pop {
  emitterSpawnScript [
   &InitializeAudioPlayer(Audio Player=DataInterface() {propertyValues {
      Attenuation=null
      Concurrency=null
      MaxPlaysPerTick=10
      SoundToPlay=/CosmeticCompanions/Assets/Audio_Shared/Interaction/Petting/SC_Mimosa_Interaction_React_Default
      bAllowLoopingOneShotSounds=false
      bLimitPlaysPerTick=true
      bOnlyActiveDuringGameplay=false
      bStopWhenComponentIsDestroyed=true
      ConfigurationUserParameter {parameter {name=None, typeDefHandle {registeredTypeIndex=1}, varData []}}
      ParameterNames []
     }})
   /Niagara/Modules/Audio/PlayAudio()
  ]
  emitterUpdateScript [
   &EmitterState(Life Cycle Mode=::NewEnumerator1, Loop Behavior=::NewEnumerator1, Loop Duration=2, UseLoopDelay=false)
   &SpawnBurst_Instantaneous(Spawn Time=$System.PopTiming, Use Spawn Probability=false, UseLoopCountLimit=false)
   &SpawnBurst_Instantaneous(Spawn Count=2, Spawn Group=1, Use Spawn Probability=false, UseLoopCountLimit=false) {module=SpawnBurst_Instantaneous001}
  ]
  particleSpawnScript [
   /SetVariables(
    Particles.GlitchColor:LinearColor=<rgba 5 1.0838 0 1>
    Particles.GlitchSize:Vector2f=<5 30>
    Particles.VisibilityTag=&SpawnGroup()
   ) {isSetParameters=true}
   &InitializeParticle(
    Color=<rgba 1 0.5825 0 1>
    Lifetime=&SelectFloatFromArray(Direct Array Index=&SpawnGroup(), Float Selection Array=DataInterface() {propertyValues {FloatData [0.5,0.5]}})
    Mesh Scale Mode=::NewEnumerator3
    Mesh Uniform Scale=0.8
    Position Offset=$System.BurstPosition
    Sprite Size Mode=::NewEnumerator3
    Uniform Sprite Size=40
    UsePositionOffset=true
    Write Material Random=true
   )
   &DynamicMaterialParameters(
    Index 0 Param 1=&RandomRangeFloat(Minimum=0, Maximum=1)
    Index 0 Param 2=&RandomRangeFloat(Minimum=8, Maximum=12)
    Index 0 Param 3=&FloatFromCurve(FloatCurve=DataInterface() {propertyValues {Curve {keys [{interpMode=RCIM_Cubic, tangentMode=RCTM_User, time=0, value=1.6}, {arriveTangent=7.3367, tangentMode=RCTM_User, time=1, value=5}]}}})
    Module.Index0Float0Write=true
    Module.Index0Float1Write=true
    Module.Index0Float2Write=true
    Module.Index0Float3Write=true
   )
   /Game/Effects/Niagara/Modules/Sprite/FN_SpriteUV(End Frame=3, Write to Module Namepsace=true)
   &CameraOffset(Camera Offset Amount=10)
   /SetVariables(Particles.CollapseCameraOffset=-25) {isSetParameters=true}
  ]
  particleUpdateScript [
   &ParticleState()
   &ScaleMeshSize(Scale Factor=&Multiply_VectorByFloat(
     Vector=<1 1 1>
     Float=&FloatFromCurve(
      FloatCurve=DataInterface() {propertyValues {Curve {
         defaultValue=1
         keys [
          {interpMode=RCIM_Cubic, tangentMode=RCTM_User, time=0, value=0}
          {
           arriveTangentWeight=0.05
           interpMode=RCIM_Cubic
           leaveTangent=0.0173
           leaveTangentWeight=0.5439
           tangentMode=RCTM_Break
           tangentWeightMode=RCTWM_WeightedBoth
           time=0.15
           value=1
          }
          <at 1 0>
         ]
        }}}
      Scale Curve=$System.FE_Scale
     )
    ))
   &ScaleSpriteSize(
    Uniform Curve Scale=$System.FE_Scale
    Uniform Curve Sprite Scale=DataInterface() {propertyValues {Curve {defaultValue=0, keys [{arriveTangent=3.3458, interpMode=RCIM_Cubic, leaveTangent=3.3458, tangentMode=RCTM_User, time=0, value=1}, <at 1 3>]}}}
   )
   &ScaleColor(
    Scale Alpha=&FloatFromCurve(FloatCurve=DataInterface() {propertyValues {Curve {
        defaultValue=1
        keys [
         {arriveTangent=0.3137, interpMode=RCIM_Cubic, leaveTangent=0.3137, tangentMode=RCTM_User, time=0, value=1}
         {arriveTangent=-2.0958, arriveTangentWeight=0.3333, interpMode=RCIM_Cubic, leaveTangent=-2.0958, tangentMode=RCTM_User, time=1, value=0}
        ]
       }}})
    Scale RGB=&VectorFromCurve(VectorCurve=DataInterface() {propertyValues {
       XCurve {defaultValue=1, keys [{interpMode=RCIM_Cubic, tangentMode=RCTM_User, time=0, value=1}, <at 1 0>]}
       YCurve {defaultValue=1, keys [{interpMode=RCIM_Cubic, tangentMode=RCTM_User, time=0, value=1}, <at 1 0>]}
       ZCurve {defaultValue=1, keys [{interpMode=RCIM_Cubic, tangentMode=RCTM_User, time=0, value=1}, <at 1 0>]}
      }})
    ScaleA=true
    ScaleRGB=true
   )
   &SolveForcesAndVelocity(Clamp Velocity=false, Limit Acceleration=false)
   /SetVariables(
    Particles.AuraColor:LinearColor=&Multiply_LinearColorByFloat(
     Linear Color=$Particles.Initial.Color
     Float=&FloatFromCurve(FloatCurve=DataInterface() {propertyValues {Curve {keys [
          {interpMode=RCIM_Cubic, tangentMode=RCTM_User, time=0, value=0}
          {interpMode=RCIM_Cubic, tangentMode=RCTM_Break, time=0.1, value=2}
          {interpMode=RCIM_Cubic, tangentMode=RCTM_User, time=0.6857, value=0.4}
          <at 0.9903 5>
         ]}}})
    )
    Particles.AuraSize:Vector2f=&Multiply_Vector2D(
     A=$Particles.Initial.SpriteSize
     B=&Vector2DFromCurve(
      Vector2Curve=DataInterface() {propertyValues {
        XCurve {keys [
          {arriveTangent=8.8756, interpMode=RCIM_Cubic, leaveTangent=8.8756, tangentMode=RCTM_User, time=0, value=1}
          {
           arriveTangent=0.6379
           arriveTangentWeight=0.1658
           interpMode=RCIM_Cubic
           leaveTangent=0.0189
           leaveTangentWeight=0.3632
           tangentMode=RCTM_Break
           tangentWeightMode=RCTWM_WeightedBoth
           time=0.3
           value=3.5
          }
          {
           arriveTangent=-0.1745
           arriveTangentWeight=0.7501
           interpMode=RCIM_Cubic
           leaveTangent=-0.1745
           leaveTangentWeight=0.7501
           tangentMode=RCTM_User
           tangentWeightMode=RCTWM_WeightedBoth
           time=1
           value=0
          }
         ]}
        YCurve {keys [
          {arriveTangent=8.8756, interpMode=RCIM_Cubic, leaveTangent=8.8756, tangentMode=RCTM_User, time=0, value=1}
          {
           arriveTangent=0.6379
           arriveTangentWeight=0.1658
           interpMode=RCIM_Cubic
           leaveTangent=0.0189
           leaveTangentWeight=0.3632
           tangentMode=RCTM_Break
           tangentWeightMode=RCTWM_WeightedBoth
           time=0.3
           value=3.5
          }
          {
           arriveTangent=-0.1745
           arriveTangentWeight=0.7501
           interpMode=RCIM_Cubic
           leaveTangent=-0.1745
           leaveTangentWeight=0.7501
           tangentMode=RCTM_User
           tangentWeightMode=RCTWM_WeightedBoth
           time=1
           value=0
          }
         ]}
       }}
      ScaleCurve=&Multiply_Vector2D_ByFloat(Float=$System.FE_Scale, Vector2D=<0.7 0.7>)
     )
    )
    Particles.GlowColor=&Multiply_LinearColorByFloat(Linear Color=&MultiplyLinearColor(Linear Color=$Particles.Color, Scale Factor=<1 1 1 1>), Float=0.33)
    Particles.RaysColor:LinearColor=&Multiply_LinearColorByFloat(
     Linear Color=&MultiplyLinearColor(
      Linear Color=$Particles.Initial.Color
      Scale Factor=&Vector4FromCurve(Vector4Curve=DataInterface() {propertyValues {
         AlphaCurve {keys [<at 0 1>,<at 1 0>]}
         BlueCurve {keys [{interpMode=RCIM_Cubic, tangentMode=RCTM_User, time=0.6, value=1}, {arriveTangent=18.8498, tangentMode=RCTM_User, time=1, value=5}]}
         GreenCurve {keys [{interpMode=RCIM_Cubic, tangentMode=RCTM_User, time=0.6, value=1}, {arriveTangent=18.8498, tangentMode=RCTM_User, time=1, value=5}]}
         RedCurve {keys [{interpMode=RCIM_Cubic, tangentMode=RCTM_User, time=0.6, value=1}, {arriveTangent=18.8498, tangentMode=RCTM_User, time=1, value=5}]}
        }})
     )
     Float=1
    )
   ) {isSetParameters=true}
   /SetVariables(
    Particles.Glitch DynP:Vector4f=&MakeVector4(
     X=&RandomRangeFloat(Minimum=0, Maximum=1)
     Y=&Lerp_Float(A=0.5, B=1, Alpha=$Particles.NormalizedAge)
     Z=/Niagara/DynamicInputs/Integer/V2/Round_Float(Float:NiagaraFloat=&FloatFromCurve(
       FloatCurve=DataInterface() {propertyValues {Curve {defaultValue=1, keys [<at 0 1>,<at 0.2035 0>,<at 0.444 0.9782>,<at 0.7 0>,<at 1 0.25>]}}}
       Scale Curve=4
      ))
     W=-1
    )
    Particles.GlitchPosition:Vector3f=$System.IntroPositionOffset
   ) {isSetParameters=true}
   /SetVariables(
    Particles.CubeCollapseColor:LinearColor=&Multiply_LinearColorByFloat(
     Linear Color=&FNTimeOfDayColor(Night Color=<rgba 0.7517 1 0.0178 1>, Morning Color=<rgba 1 1 1 1>, Day Color=<rgba 1 1 1 1>, Evening Color=<rgba 0.8199 1 0 1>)
     Float=&FloatFromCurve(FloatCurve=DataInterface() {propertyValues {Curve {keys [<at 0 1>,<at 1 1>]}}}, Scale Curve=$System.FE_Scale)
    )
    Particles.CubeCollapseSize:Vector2f=&Multiply_Vector2D_ByFloat(
     Float=&FloatFromCurve(
      FloatCurve=DataInterface() {propertyValues {Curve {keys [
          {arriveTangent=6.3629, interpMode=RCIM_Cubic, leaveTangent=6.3629, tangentMode=RCTM_User, time=0, value=0.1}
          {interpMode=RCIM_Cubic, tangentMode=RCTM_User, time=0.1814, value=0.8806}
          <at 1 0>
         ]}}}
      Scale Curve=$System.FE_Scale
     )
     Vector2D=<200 200>
    )
   ) {isSetParameters=true}
  ]
  rendererClasses [&NiagaraSpriteRendererProperties, &NiagaraMeshRendererProperties]
  renderers [
   {rendererClass=&NiagaraSpriteRendererProperties}
   {rendererClass=&NiagaraMeshRendererProperties, rendererIndex=1}
   {rendererClass=&NiagaraSpriteRendererProperties, rendererIndex=2}
   {rendererClass=&NiagaraSpriteRendererProperties, rendererIndex=3}
   {rendererClass=&NiagaraSpriteRendererProperties, rendererIndex=4}
  ]
 }
}
summary {
 systemName=NS_ReactFX_CosmoHound
 emitters [
  {emitterName=Pop, rendererClasses [&NiagaraSpriteRendererProperties, &NiagaraMeshRendererProperties]}
  {emitterName=Cubes, rendererClasses [&NiagaraMeshRendererProperties]}
  {emitterName=Main_Heart, rendererClasses [&NiagaraMeshRendererProperties, &NiagaraSpriteRendererProperties]}
  {bEnabled=false, emitterName=Lightning, rendererClasses [&NiagaraMeshRendererProperties]}
  {bEnabled=false, emitterName=Light, rendererClasses [&NiagaraLightRendererProperties]}
  {bEnabled=false, emitterName=Glow_And_Light, rendererClasses [&NiagaraLightRendererProperties]}
  {bEnabled=false, emitterName=Cubes001, rendererClasses [&NiagaraMeshRendererProperties]}
 ]
 userVariables [
  User.bIsFrontEnd {type=NiagaraBool}
  User.bIsFrontEndPreview {type=NiagaraBool}
 ]
}
