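# CrescentNail reaction effect for cosmetic companions: a Niagara sprite burst with a shared petting-interaction sound.
# Intro moves a growing sprite toward the system burst position over IntroLength, then Pop, Sparkles and spinning, falling Eggs burst there.
# Eggs are tinted by time of day.
# Condensed_LOW rebuilds these stages in one emitter, using spawn groups and mask inputs to select per-stage behavior, apparently as a cheaper variant.
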NiagaraSystem /CosmeticCompanions/FX/React/NS_ReactFX_CrescentNail
dependencies {
 usedDataInterfaces [
  &NiagaraDataInterfaceAudioPlayer
  &NiagaraDataInterfaceCurve
  &NiagaraDataInterfaceVectorCurve
  &NiagaraDataInterfaceSpriteRendererInfo
 ]
 usedDynamicInputs [/Niagara/DynamicInputs/TypeConversions/MakeCustomPositionFromBool, &FN_FrontEndVector3Modifier]
 usedModules [
  &self:SystemSpawnScript.SystemScriptSource.SystemScriptGraph.NiagaraNodeAssignment_0.SetVariables_#1
  &self:SystemSpawnScript.SystemScriptSource.SystemScriptGraph.NiagaraNodeAssignment_1.SetVariables_#2
  &self:SystemSpawnScript.SystemScriptSource.SystemScriptGraph.NiagaraNodeAssignment_5.SetVariables_#3
  &self:SystemSpawnScript.SystemScriptSource.SystemScriptGraph.NiagaraNodeAssignment_4.SetVariables_#4
  &self:SystemSpawnScript.SystemScriptSource.SystemScriptGraph.NiagaraNodeAssignment_6.SetVariables_#5
  &self:SystemSpawnScript.SystemScriptSource.SystemScriptGraph.NiagaraNodeAssignment_2.SetVariables_#6
  &self:SystemSpawnScript.SystemScriptSource.SystemScriptGraph.NiagaraNodeAssignment_3.SetVariables_#7
  &SystemState
 ]
}
emitters {
 Condensed_LOW {
  emitterSpawnScript [
   &InitializeAudioPlayer(Audio Player=DataInterface() {propertyValues {
      Attenuation=null
      Concurrency=null
      MaxPlaysPerTick=10
      SoundToPlay=/CosmeticCompanions/Assets/Audio_Shared/Interaction/Petting/SC_Mimosa_Interaction_React_Default
      bAllowLoopingOneShotSounds=false
      bLimitPlaysPerTick=true
      bOnlyActiveDuringGameplay=false
      bStopWhenComponentIsDestroyed=true
      ConfigurationUserParameter {parameter {name=None, typeDefHandle {registeredTypeIndex=1}, varData []}}
      ParameterNames []
     }})
   /Niagara/Modules/Audio/PlayAudio() {module=PlayAudio001}
  ]
  emitterUpdateScript [
   &EmitterState()
   &SpawnBurst_Instantaneous(Use Spawn Probability=false, UseLoopCountLimit=false)
   &SpawnBurst_Instantaneous(
    Spawn Group=1
    Spawn Time=&Subtract_Float(A=$System.IntroLength, B=0.1)
    Use Spawn Probability=false
    UseLoopCountLimit=false
   ) {module=SpawnBurst_Instantaneous001}
   &SpawnBurst_Instantaneous(
    Spawn Count=3
    Spawn Group=2
    Spawn Time=&Subtract_Float(A=$System.IntroLength, B=0.1)
    Use Spawn Probability=false
    UseLoopCountLimit=false
   ) {module=SpawnBurst_Instantaneous002}
   &SpawnBurst_Instantaneous(
    Spawn Count=14
    Spawn Group=3
    Spawn Time=&Subtract_Float(A=$System.IntroLength, B=0.1)
    Use Spawn Probability=true
    UseLoopCountLimit=false
   ) {module=SpawnBurst_Instantaneous003}
  ]
  particleSpawnScript [
   &InitializeParticle(
    Color=&MaskLinearColor(
     Mask Source=0
     Mask Parameter=?
     Mask=0
     Masked Linear Color=<rgba 1 1 1 1>
     Passthrough Linear Color=&MaskLinearColor(
      Mask Source=0
      Mask Parameter=?
      Mask=1
      Masked Linear Color=<rgba 1 1 1 1>
      Passthrough Linear Color=&MaskLinearColor(
       Mask Source=0
       Mask Parameter=?
       Mask=2
       Masked Linear Color=&MaskLinearColor(
        Mask Source=0
        Mask Parameter=?
        Mask=2
        Masked Linear Color=/Niagara/DynamicInputs/Add/Add_LinearColor(
         A:LinearColor=&Multiply_LinearColorByFloat(
          Linear Color=&RandomRangeLinearColor(Minimum=<rgba 0.5104 0.5104 0.5104 1>, Maximum=<rgba 0.2865 0.2865 0.2865 1>)
          Float=&FNTimeOfDayScalar(Night Scale=0.2, Morning Scale=1, Evening Scale=1)
         )
         B:LinearColor=&FNTimeOfDayColor(
          Night Color=<rgba 0.0223 0.0264 0.25 1>
          Morning Color=<rgba 0.0365 0.0082 0.0061 1>
          Day Color=<rgba 0 0 0 1>
          Evening Color=<rgba 0.0208 0.0146 0.0129 1>
         )
        )
        Passthrough Linear Color=<rgba 1 1 1 2>
       )
       Passthrough Linear Color=<rgba 1 1 1 1>
      )
     )
    )
    Lifetime=&MaskFloat(
     Mask Source=0
     Mask Parameter=?
     Mask=0
     Masked Float=$System.IntroLength
     Passthrough Float=&MaskFloat(
      Mask Source=0
      Mask Parameter=?
      Mask=1
      Masked Float=0.2
      Passthrough Float=&MaskFloat(
       Mask Source=0
       Mask Parameter=?
       Mask=2
       Masked Float=&RandomRangeFloat(Minimum=0.75, Maximum=1)
       Passthrough Float=&MaskFloat(
        Mask Source=0
        Mask Parameter=?
        Mask=3
        Masked Float=&UniformRangedFloat(Minimum=0.3, Maximum=1)
        Passthrough Float=$System.IntroLength
       )
      )
     )
    )
    Sprite Rotation Mode=::NewEnumerator0
    Sprite Size Mode=::NewEnumerator3
    Uniform Sprite Size=15
    UsePositionOffset=false
    Write Material Random=true
   )
   /SetVariables(
    Particles.PopDynParam:Vector4f=<0 0 0 1>
    Particles.PopPos:Vector3f=$System.BurstPosition
    Particles.VisibilityTag=&SpawnGroup()
   ) {isSetParameters=true}
   &SubUVAnimation(
    Evaluation Type=::NewEnumerator0
    Output Alternate Renderer Binding=true
    "Sprite Renderer "=DataInterface() {propertyValues {SpriteRenderer=&self:Condensed_LOW_0.NiagaraSpriteRendererProperties_4}}
    SubUV Animation Mode:ENiagara_SubUVLookupModeV2=::NewEnumerator2
    UseEndFrame=false
    UseStartFrame=false
   )
   &ShapeLocation(
    MaskSpawnGroup=false
    Override Seed=false
    Sphere Radius=&MaskFloat(Mask Source=0, Mask Parameter=?, Mask=2, Masked Float=10, Passthrough Float=0)
    UseOwnerScale=true
   )
   &AddVelocity(
    Cone Axis=<0 0 1>
    Module.UseConeFalloff=false
    Module.UseVelDistribution=false
    Override Seed=false
    Velocity Mode=::NewEnumerator2
    Velocity Speed=&MaskFloat(
     Mask Source=0
     Mask Parameter=?
     Mask=2
     Masked Float=&RandomRangeFloat(Minimum=75, Maximum=125, Random Seed=&ReturnParticleIDAttributes())
     Passthrough Float=0
    )
   )
   &AddVelocity(
    Velocity=&UniformRangedVector(Minimum=<-200 -200 -200>, Maximum=<200 200 200>)
    Velocity Speed Scale=&MaskFloat(Mask Source=0, Mask Parameter=?, Mask=3, Masked Float=1, Passthrough Float=0)
   ) {module=AddVelocity001}
  ]
  particleUpdateScript [
   &ParticleState()
   &SpriteRotationRate(Rotation Rate=&MaskFloat(Mask Source=0, Mask Parameter=?, Mask=2, Masked Float=360, Passthrough Float=0))
   /SetVariables(
    Particles.Glow Color=&Multiply_LinearColorByFloat(Linear Color=$System.Layer 2 Color, Float=0.3)
    Particles.Glow Size=&Multiply_Vector2D(A=$Particles.SpriteSize, B=<0.6 0.6>)
   ) {isSetParameters=true, module=SetVariables_#8}
   &GravityForce(Gravity=&MaskVector(
     Mask Source=0
     Mask Parameter=?
     Mask=2
     Masked Vector=&RandomRangeVector(Minimum=<0 0 -200>, Maximum=<0 0 -500>)
     Passthrough Vector=&MaskVector(
      Mask Source=0
      Mask Parameter=?
      Mask=3
      Masked Vector=&RandomRangeVector(Minimum=<-100 -100 -200>, Maximum=<100 100 200>)
      Passthrough Vector=<0 0 0>
     )
    ))
   &OffsetPosition(Position Offset=&MaskVector(
     Mask Source=0
     Mask Parameter=?
     Mask=0
     Masked Vector=&Lerp_Vector(
      A=$System.HeartIntroSpawnPosition
      B=$System.BurstPosition
      Alpha=&FloatFromCurve(FloatCurve=DataInterface() {propertyValues {Curve {keys [
           {arriveTangent=0.2549, interpMode=RCIM_Cubic, leaveTangent=0.2549, tangentMode=RCTM_User, time=0, value=0.5}
           {arriveTangent=0.5, leaveTangent=0.5, time=0.8, value=1}
           {arriveTangent=-0.006, time=1, value=1}
          ]}}})
     )
     Passthrough Vector=$System.BurstPosition
    ))
   &ScaleColor(
    Scale Alpha=&MaskFloat(
     Mask Source=0
     Mask Parameter=?
     Mask=0
     Masked Float=&FloatFromCurve(FloatCurve=DataInterface() {propertyValues {Curve {
         defaultValue=0
         keys [
          {arriveTangent=9.0725, interpMode=RCIM_Cubic, leaveTangent=9.0725, tangentMode=RCTM_User, time=0, value=0}
          {arriveTangent=-0.2441, leaveTangent=-0.2441, time=0.1408, value=0.9847}
          {arriveTangent=-1.146, leaveTangent=-1.146, time=0.9, value=1}
          {arriveTangent=-0.0405, interpMode=RCIM_Cubic, leaveTangent=-0.0405, tangentMode=RCTM_User, time=1, value=0}
         ]
        }}})
     Passthrough Float=&MaskFloat(
      Mask Source=0
      Mask Parameter=?
      Mask=1
      Masked Float=&FloatFromCurve(FloatCurve=DataInterface() {propertyValues {Curve {
          defaultValue=0
          keys [
           {arriveTangent=0.3137, interpMode=RCIM_Cubic, leaveTangent=0.3137, tangentMode=RCTM_User, time=0, value=1}
           {arriveTangent=-2.0958, interpMode=RCIM_Cubic, leaveTangent=-2.0958, tangentMode=RCTM_User, time=1, value=0}
          ]
         }}})
      Passthrough Float=&MaskFloat(
       Mask Source=0
       Mask Parameter=?
       Mask=2
       Masked Float=1
       Passthrough Float=&MaskFloat(
        Mask Source=0
        Mask Parameter=?
        Mask=3
        Masked Float=&FloatFromCurve(FloatCurve=DataInterface() {propertyValues {Curve {keys [<at -0.0092 0.0144>,<at 0.0967 0.9767>,<at 1.0011 0.002>]}}})
        Passthrough Float=1
       )
      )
     )
    )
    Scale RGB=&MaskVector(
     Mask Source=0
     Mask Parameter=?
     Mask=0
     Masked Vector=<1 1 1>
     Passthrough Vector=&MaskVector(
      Mask Source=0
      Mask Parameter=?
      Mask=1
      Masked Vector=&VectorFromCurve(VectorCurve=DataInterface() {propertyValues {
         XCurve {keys [{interpMode=RCIM_Cubic, tangentMode=RCTM_User, time=0, value=1}, {arriveTangent=-0.2294, tangentMode=RCTM_User, time=1, value=0}]}
         YCurve {keys [{interpMode=RCIM_Cubic, tangentMode=RCTM_User, time=0, value=1}, {arriveTangent=-0.2294, tangentMode=RCTM_User, time=1, value=0}]}
         ZCurve {keys [{interpMode=RCIM_Cubic, tangentMode=RCTM_User, time=0, value=1}, {arriveTangent=-0.2294, tangentMode=RCTM_User, time=1, value=0}]}
        }})
      Passthrough Vector=&MaskVector(
       Mask Source=0
       Mask Parameter=?
       Mask=2
       Masked Vector=<1 1 1>
       Passthrough Vector=&MaskVector(Mask Source=0, Mask Parameter=?, Mask=3, Masked Vector=<2 2 2>, Passthrough Vector=<1 1 1>)
      )
     )
    )
    ScaleA=true
    ScaleRGB=true
   )
   &ScaleSpriteSize(
    Scale Sprite Size Mode=::NewEnumerator0
    Uniform Scale Factor=&MaskFloat(
     Mask Source=0
     Mask Parameter=?
     Mask=0
     Masked Float=&Add_Float(
      A=0.5
      B=&FloatFromCurve(FloatCurve=DataInterface() {propertyValues {Curve {
          defaultValue=0
          keys [
           {arriveTangent=0.8193, interpMode=RCIM_Cubic, leaveTangent=0.8193, tangentMode=RCTM_User, time=0, value=0.2}
           {arriveTangent=2.055, interpMode=RCIM_Cubic, leaveTangent=2.055, tangentMode=RCTM_User, time=0.7247, value=0.9545}
           {arriveTangent=-0.9053, interpMode=RCIM_Cubic, leaveTangent=-0.9053, tangentMode=RCTM_User, time=0.8783, value=3}
           {arriveTangent=-0.1202, interpMode=RCIM_Cubic, leaveTangent=-0.1202, tangentMode=RCTM_User, time=1, value=1}
          ]
         }}})
     )
     Passthrough Float=&MaskFloat(
      Mask Source=0
      Mask Parameter=?
      Mask=1
      Masked Float=&FloatFromCurve(FloatCurve=DataInterface() {propertyValues {Curve {keys [
           {arriveTangent=3.3458, interpMode=RCIM_Cubic, leaveTangent=3.3458, tangentMode=RCTM_User, time=0, value=1}
           {arriveTangent=1.251, tangentMode=RCTM_User, time=1, value=3}
          ]}}})
      Passthrough Float=&MaskFloat(
       Mask Source=0
       Mask Parameter=?
       Mask=2
       Masked Float=&FloatFromCurve(FloatCurve=DataInterface() {propertyValues {Curve {keys [{interpMode=RCIM_Cubic, tangentMode=RCTM_User, time=0, value=1}, {arriveTangent=-2.7673, tangentMode=RCTM_User, time=1, value=0}]}}})
       Passthrough Float=&MaskFloat(
        Mask Source=0
        Mask Parameter=?
        Mask=3
        Masked Float=&FloatFromCurve(FloatCurve=DataInterface() {propertyValues {Curve {keys [<at -0.0046 0>,<at 0.1746 0.9952>,<at 1.0011 0.002>]}}})
        Passthrough Float=1
       )
      )
     )
    )
   ) {module=ScaleSpriteSize002}
   &Drag(
    Drag=&MergeHelper_FloatSwitch(
     0=0
     1=0
     2=0
     3=&FloatFromCurve(FloatCurve=DataInterface() {propertyValues {Curve {keys [{interpMode=RCIM_Cubic, tangentMode=RCTM_User, time=0, value=6}, {arriveTangent=-4.5401, tangentMode=RCTM_User, time=1, value=2}]}}})
     Index=&SpawnGroup()
     Num=::NewEnumerator2
    )
    Rotational Drag=0
    Use Linear Drag=true
    Use Rotational Drag=true
   )
   &SolveForcesAndVelocity(Clamp Velocity=false, Limit Acceleration=false)
   /SetVariables(Particles.Heart DynParams:Vector4f=&MakeVector4(X=1, Y=1, Z=0, W=0)) {isSetParameters=true}
  ]
  rendererClasses [&NiagaraSpriteRendererProperties]
  renderers [
   {rendererClass=&NiagaraSpriteRendererProperties}
   {rendererClass=&NiagaraSpriteRendererProperties, rendererIndex=1}
   {rendererClass=&NiagaraSpriteRendererProperties, rendererIndex=2}
   {rendererClass=&NiagaraSpriteRendererProperties, rendererIndex=3}
   {rendererClass=&NiagaraSpriteRendererProperties, rendererIndex=4}
   {rendererClass=&NiagaraSpriteRendererProperties, rendererIndex=5}
  ]
 }
 Eggs {
  emitterUpdateScript [
   &EmitterState()
   &SpawnBurst_Instantaneous(
    Spawn Count=3
    Spawn Time=&Subtract_Float(A=$System.IntroLength, B=0)
    Use Spawn Probability=false
    UseLoopCountLimit=false
   )
  ]
  particleSpawnScript [
   &InitializeParticle(
    Color Maximum=<rgba 0.2865 0.2865 0.2865 1>
    Color Minimum=<rgba 0.5104 0.5104 0.5104 1>
    Color Mode=::NewEnumerator2
    Lifetime=&RandomRangeFloat(Minimum=0.75, Maximum=1)
    Position Offset=$System.BurstPosition
    Sprite Rotation Mode=::NewEnumerator2
    Sprite Size Mode=::NewEnumerator1
    Uniform Sprite Size Max=15
    Uniform Sprite Size Min=10
    UsePositionOffset=true
    Write Material Random=true
   )
   &AddVelocity(
    Cone Axis=<0 0 1>
    Module.UseConeFalloff=false
    Module.UseVelDistribution=false
    Override Seed=false
    Velocity Mode=::NewEnumerator2
    Velocity Speed=&RandomRangeFloat(Minimum=75, Maximum=125)
   )
   &ShapeLocation(MaskSpawnGroup=false, Override Seed=false, Sphere Radius=10, UseOwnerScale=true)
   &SubUVAnimation(
    Evaluation Type=::NewEnumerator0
    "Sprite Renderer "=DataInterface() {propertyValues {SpriteRenderer=&self:Eggs_0.NiagaraSpriteRendererProperties_0}}
    SubUV Animation Mode:ENiagara_SubUVLookupModeV2=::NewEnumerator2
    UseEndFrame=false
    UseStartFrame=false
   )
   &Color(Color=&Add_LinearColor(
     A:LinearColor=&Multiply_LinearColorByFloat(
      Linear Color=$Particles.Color
      Float=&FNTimeOfDayScalar(Night Scale=0.2, Morning Scale=1, Evening Scale=1)
     )
     B:LinearColor=&FNTimeOfDayColor(
      Night Color=<rgba 0.0223 0.0264 0.25 1>
      Morning Color=<rgba 0.0365 0.0082 0.0061 1>
      Day Color=<rgba 0 0 0 1>
      Evening Color=<rgba 0.0208 0.0146 0.0129 1>
     )
    ))
  ]
  particleUpdateScript [
   &ParticleState()
   &ScaleSpriteSize(
    Scale Sprite Size Mode=::NewEnumerator0
    Uniform Scale Factor=&FloatFromCurve(FloatCurve=DataInterface() {propertyValues {Curve {keys [{interpMode=RCIM_Cubic, tangentMode=RCTM_User, time=0, value=1}, {arriveTangent=-2.7673, tangentMode=RCTM_User, time=1, value=0}]}}})
   )
   &SpriteRotationRate(Rotation Rate=360)
   &GravityForce(Gravity=&RandomRangeVector(Minimum=<0 0 -200>, Maximum=<0 0 -500>))
   &SolveForcesAndVelocity(Clamp Velocity=false, Limit Acceleration=false)
   &ScaleColor(
    Scale Alpha=&FloatFromCurve(FloatCurve=DataInterface() {propertyValues {Curve {keys [{interpMode=RCIM_Cubic, tangentMode=RCTM_User, time=0, value=1}, <at 1 1>]}}})
    Scale RGB=<0.25 0.25 0.25>
    ScaleA=true
    ScaleRGB=true
   ) {enabled=false}
  ]
  rendererClasses [&NiagaraSpriteRendererProperties]
  renderers [{rendererClass=&NiagaraSpriteRendererProperties}]
 }
 Intro {
  emitterSpawnScript [
   &InitializeAudioPlayer(Audio Player=DataInterface() {propertyValues {
      Attenuation=null
      Concurrency=null
      MaxPlaysPerTick=10
      SoundToPlay=&SC_Mimosa_Interaction_React_Default
      bAllowLoopingOneShotSounds=false
      bLimitPlaysPerTick=true
      bOnlyActiveDuringGameplay=false
      bStopWhenComponentIsDestroyed=true
      ConfigurationUserParameter {parameter {name=None, typeDefHandle {registeredTypeIndex=1}, varData []}}
      ParameterNames []
     }})
   &PlayAudio() {module=PlayAudio001}
  ]
  emitterUpdateScript [
   &EmitterState()
   &SpawnBurst_Instantaneous(Use Spawn Probability=false, UseLoopCountLimit=false)
  ]
  particleSpawnScript [&InitializeParticle(
    Lifetime=$System.IntroLength
    Sprite Rotation Mode=::NewEnumerator0
    Sprite Size Mode=::NewEnumerator3
    Uniform Sprite Size=15
    UsePositionOffset=false
    Write Material Random=true
   )]
  particleUpdateScript [
   &ParticleState()
   &ScaleSpriteSize(
    Scale Sprite Size Mode=::NewEnumerator0
    Uniform Scale Factor=&Add_Float(
     A=0.5
     B=&FloatFromCurve(FloatCurve=DataInterface() {propertyValues {Curve {
         defaultValue=0
         keys [
          {arriveTangent=0.8193, interpMode=RCIM_Cubic, leaveTangent=0.8193, tangentMode=RCTM_User, time=0, value=0.2}
          {arriveTangent=2.055, interpMode=RCIM_Cubic, leaveTangent=2.055, tangentMode=RCTM_User, time=0.7247, value=0.9545}
          {arriveTangent=-0.9053, interpMode=RCIM_Cubic, leaveTangent=-0.9053, tangentMode=RCTM_User, time=0.8783, value=3}
          {arriveTangent=-0.1202, interpMode=RCIM_Cubic, leaveTangent=-0.1202, tangentMode=RCTM_User, time=1, value=1}
         ]
        }}})
    )
   ) {module=ScaleSpriteSize002}
   &ScaleColor(
    Scale Alpha=&FloatFromCurve(FloatCurve=DataInterface() {propertyValues {Curve {
        defaultValue=0
        keys [
         {arriveTangent=9.0725, interpMode=RCIM_Cubic, leaveTangent=9.0725, tangentMode=RCTM_User, time=0, value=0}
         {arriveTangent=-0.2441, leaveTangent=-0.2441, time=0.1408, value=0.9847}
         {arriveTangent=-1.146, leaveTangent=-1.146, time=0.9, value=1}
         {arriveTangent=-0.0405, interpMode=RCIM_Cubic, leaveTangent=-0.0405, tangentMode=RCTM_User, time=1, value=0}
        ]
       }}})
    ScaleA=true
    ScaleRGB=true
   )
   &OffsetPosition(Position Offset=&Lerp_Vector(
     A=$System.HeartIntroSpawnPosition
     B=$System.BurstPosition
     Alpha=&FloatFromCurve(FloatCurve=DataInterface() {propertyValues {Curve {keys [
          {arriveTangent=0.2549, interpMode=RCIM_Cubic, leaveTangent=0.2549, tangentMode=RCTM_User, time=0, value=0.5}
          {arriveTangent=0.5, leaveTangent=0.5, time=0.8, value=1}
          {arriveTangent=-0.006, time=1, value=1}
         ]}}})
    ))
   &DynamicMaterialParameters(
    Index 0 Param 1=&FloatFromCurve(FloatCurve=DataInterface() {propertyValues {Curve {keys [{arriveTangent=0.6296, interpMode=RCIM_Cubic, leaveTangent=0.6296, tangentMode=RCTM_User, time=0, value=0}, <at 0.75 0.9676>, <at 1 1>]}}})
    Index 0 Param 2=1
    Module.Index0Float0Write=true
    Module.Index0Float1Write=true
    Module.Index0Float2Write=true
    Module.Index0Float3Write=true
   )
   /SetVariables(
    Particles.Glow Color=&Multiply_LinearColorByFloat(Linear Color=$System.Layer 2 Color, Float=0.3)
    Particles.Glow Size=&Multiply_Vector2D(A=$Particles.SpriteSize, B=<0.6 0.6>)
   ) {isSetParameters=true, module=SetVariables_#8}
  ]
  rendererClasses [&NiagaraSpriteRendererProperties]
  renderers [
   {rendererClass=&NiagaraSpriteRendererProperties}
   {rendererClass=&NiagaraSpriteRendererProperties, rendererIndex=1}
  ]
 }
 Pop {
  emitterUpdateScript [
   &EmitterState()
   &SpawnBurst_Instantaneous(Spawn Time=&Subtract_Float(A=$System.IntroLength, B=0), Use Spawn Probability=false, UseLoopCountLimit=false)
  ]
  particleSpawnScript [
   &InitializeParticle(
    Color=<rgba 1 1 1 0.8>
    Lifetime=0.2
    Position Offset=$System.BurstPosition
    Sprite Rotation Mode=::NewEnumerator0
    Sprite Size Mode=::NewEnumerator3
    Uniform Sprite Size=15
    UsePositionOffset=true
    Write Material Random=true
   )
   &DynamicMaterialParameters(
    Index 0 Param 4=1
    Module.Index0Float0Write=true
    Module.Index0Float1Write=true
    Module.Index0Float2Write=true
    Module.Index0Float3Write=true
   )
  ]
  particleUpdateScript [
   &ParticleState()
   &ScaleSpriteSize(
    Scale Sprite Size Mode=::NewEnumerator0
    Uniform Scale Factor=&FloatFromCurve(FloatCurve=DataInterface() {propertyValues {Curve {keys [
         {arriveTangent=3.3458, interpMode=RCIM_Cubic, leaveTangent=3.3458, tangentMode=RCTM_User, time=0, value=1}
         {arriveTangent=1.251, tangentMode=RCTM_User, time=1, value=3}
        ]}}})
   ) {module=ScaleSpriteSize002}
   &ScaleColor(
    Scale Alpha=&FloatFromCurve(FloatCurve=DataInterface() {propertyValues {Curve {
        defaultValue=0
        keys [
         {arriveTangent=0.3137, interpMode=RCIM_Cubic, leaveTangent=0.3137, tangentMode=RCTM_User, time=0, value=1}
         {arriveTangent=-2.0958, interpMode=RCIM_Cubic, leaveTangent=-2.0958, tangentMode=RCTM_User, time=1, value=0}
        ]
       }}})
    Scale RGB=&VectorFromCurve(VectorCurve=DataInterface() {propertyValues {
       XCurve {keys [{interpMode=RCIM_Cubic, tangentMode=RCTM_User, time=0, value=1}, {arriveTangent=-0.2294, tangentMode=RCTM_User, time=1, value=0}]}
       YCurve {keys [{interpMode=RCIM_Cubic, tangentMode=RCTM_User, time=0, value=1}, {arriveTangent=-0.2294, tangentMode=RCTM_User, time=1, value=0}]}
       ZCurve {keys [{interpMode=RCIM_Cubic, tangentMode=RCTM_User, time=0, value=1}, {arriveTangent=-0.2294, tangentMode=RCTM_User, time=1, value=0}]}
      }})
    ScaleA=true
    ScaleRGB=true
   )
   &SolveForcesAndVelocity(Clamp Velocity=false, Limit Acceleration=false)
   /SetVariables(Particles.Glow Color=&Multiply_LinearColorByFloat(
     Linear Color=&MultiplyLinearColor(
      Linear Color=$Particles.Color
      Scale Factor=&MakeVector4_FromLinearColor(LinearColor=<rgba 1 0.5877 0.8563 1>)
     )
     Float=0.33
    )) {isSetParameters=true}
  ]
  rendererClasses [&NiagaraSpriteRendererProperties]
  renderers [
   {rendererClass=&NiagaraSpriteRendererProperties}
   {rendererClass=&NiagaraSpriteRendererProperties, rendererIndex=1}
  ]
 }
 Sparkles {
  emitterUpdateScript [
   &EmitterState()
   &SpawnBurst_Instantaneous(Spawn Count=20, Spawn Time=$System.IntroLength, Use Spawn Probability=false)
  ]
  particleSpawnScript [
   &InitializeParticle(
    Color=<rgba 5 3.7043 1.51 1>
    Lifetime=&UniformRangedFloat(Minimum=0.3, Maximum=1)
    Position Offset=$System.BurstPosition
    Sprite Size Mode=::NewEnumerator3
    Uniform Sprite Size=&UniformRangedFloat(Minimum=4, Maximum=8)
    UsePositionOffset=true
    Write Material Random=true
   ) {module=InitializeParticle001}
   &AddVelocity(Velocity=&UniformRangedVector(Minimum=<-250 -250 -250>, Maximum=<250 250 250>))
   &ApplyInitialForces()
  ]
  particleUpdateScript [
   &ParticleState()
   &GravityForce(Gravity=&RandomRangeVector(Minimum=<-100 -100 -200>, Maximum=<100 100 200>))
   &Drag(
    Drag=&FloatFromCurve(FloatCurve=DataInterface() {propertyValues {Curve {keys [{interpMode=RCIM_Cubic, tangentMode=RCTM_User, time=0, value=6}, {arriveTangent=-4.5401, tangentMode=RCTM_User, time=1, value=2}]}}})
    Use Linear Drag=true
    Use Rotational Drag=true
   )
   &SolveForcesAndVelocity(Clamp Velocity=false, Limit Acceleration=false)
   &ScaleColor(
    Scale Alpha=&FloatFromCurve(FloatCurve=DataInterface() {propertyValues {Curve {keys [<at -0.0092 0.0144>,<at 0.0967 0.9767>,<at 1.0011 0.002>]}}})
    ScaleA=true
    ScaleRGB=true
   )
   &ScaleSpriteSize(Scale Factor=&Vector2DFromFloat(Value=&FloatFromCurve(FloatCurve=DataInterface() {propertyValues {Curve {keys [<at -0.0046 0>,<at 0.1746 0.9952>,<at 1.0011 0.002>]}}})))
  ]
  rendererClasses [&NiagaraSpriteRendererProperties]
  renderers [{rendererClass=&NiagaraSpriteRendererProperties}]
 }
}
summary {
 systemName=NS_ReactFX_CrescentNail
 emitters [
  {emitterName=Intro, rendererClasses [&NiagaraSpriteRendererProperties]}
  {emitterName=Pop, rendererClasses [&NiagaraSpriteRendererProperties]}
  {emitterName=Sparkles, rendererClasses [&NiagaraSpriteRendererProperties]}
  {emitterName=Eggs, rendererClasses [&NiagaraSpriteRendererProperties]}
  {emitterName=Condensed_LOW, rendererClasses [&NiagaraSpriteRendererProperties]}
 ]
 userVariables [
  User.bIsFrontEnd {type=NiagaraBool}
  User.bIsFrontEndPreview {type=NiagaraBool}
 ]
}
