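# Niagara reaction effect for the DroneGravy cosmetic companion, pairing a root flipbook and leaf burst with a petting reaction sound.
# RootsIntro plays the sound and a curve-driven SubUV sprite; Leaf_Out later bursts tumbling, dragged leaf sprites outward.
# Both spawn around System.BurstPosition and tint by time of day, with RootsIntro switching to flat grey for front-end preview.
# BigHeart, HeartGlow and both LWE emitters show no renderers or modules.
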NiagaraSystem /CosmeticCompanions/FX/React/NS_ReactFX_DroneGravy
dependencies {
 usedDataInterfaces [
  &NiagaraDataInterfaceAudioPlayer
  &NiagaraDataInterfaceSpriteRendererInfo
  &NiagaraDataInterfaceCurve
  &NiagaraDataInterfaceArrayColor
 ]
 usedDynamicInputs [/Niagara/DynamicInputs/TypeConversions/MakeCustomPositionFromBool, &FN_FrontEndVector3Modifier]
 usedModules [
  &self:SystemSpawnScript.SystemScriptSource.SystemScriptGraph.NiagaraNodeAssignment_0.SetVariables_#1
  &self:SystemSpawnScript.SystemScriptSource.SystemScriptGraph.NiagaraNodeAssignment_1.SetVariables_#2
  &self:SystemSpawnScript.SystemScriptSource.SystemScriptGraph.NiagaraNodeAssignment_7.SetVariables_#3
  &self:SystemSpawnScript.SystemScriptSource.SystemScriptGraph.NiagaraNodeAssignment_4.SetVariables_#4
  &self:SystemSpawnScript.SystemScriptSource.SystemScriptGraph.NiagaraNodeAssignment_8.SetVariables_#5
  &self:SystemSpawnScript.SystemScriptSource.SystemScriptGraph.NiagaraNodeAssignment_2.SetVariables_#6
  &self:SystemSpawnScript.SystemScriptSource.SystemScriptGraph.NiagaraNodeAssignment_3.SetVariables_#7
  &self:SystemSpawnScript.SystemScriptSource.SystemScriptGraph.NiagaraNodeAssignment_9.SetVariables_#8
  &self:SystemSpawnScript.SystemScriptSource.SystemScriptGraph.NiagaraNodeAssignment_10.SetVariables_#9
  &self:SystemSpawnScript.SystemScriptSource.SystemScriptGraph.NiagaraNodeAssignment_5.SetVariables_#10
  &self:SystemSpawnScript.SystemScriptSource.SystemScriptGraph.NiagaraNodeAssignment_6.SetVariables_#11
  &SystemState
 ]
}
emitters {
 BigHeart {rendererClasses [], renderers []}
 HeartGlow {rendererClasses [], renderers []}
 LWE_Corona {rendererClasses [], renderers []}
 LWE_Rays {rendererClasses [], renderers []}
 Leaf_Out {
  emitterUpdateScript [
   &EmitterState(Inactive Response=::NewEnumerator1, Life Cycle Mode=::NewEnumerator1, Loop Behavior=::NewEnumerator1, Loop Duration=3, UseLoopDelay=false)
   &SpawnBurst_Instantaneous(Spawn Count=15, Spawn Time=0.98, Use Spawn Probability=false, UseLoopCountLimit=false) {module=SpawnBurst_Instantaneous001}
  ]
  particleSpawnScript [
   &InitializeParticle(
    Color=&SelectLinearColorFromArray(
     Array Sampling Mode=::NewEnumerator0
     Direct Array Index=?
     Color Selection Array=DataInterface() {propertyValues {ColorData [<rgba 0.1256 0.3854 0 1>, <rgba 0.3549 0.5677 0.004 1>]}}
     Random Seed=0
    )
    Lifetime Max=0.7
    Lifetime Min=0.5
    Lifetime Mode=::NewEnumerator1
    Position Offset=$System.BurstPosition
    Sprite Rotation Mode=::NewEnumerator2
    Sprite Size Max=<5 6>
    Sprite Size Min=<3 5>
    Sprite Size Mode=::NewEnumerator2
    UsePositionOffset=true
    Write Material Random=true
   )
   &AddVelocity(
    Module.Use Falloff Distance=false
    Override Seed=false
    Velocity Mode=::NewEnumerator1
    Velocity Origin:Vector3f=$System.BurstPosition
    Velocity Speed=&RandomRangeFloat(Minimum=90, Maximum=180)
   )
   &CameraOffset(Camera Offset Amount=&RandomRangeFloat(Minimum=0, Maximum=15))
  ]
  particleUpdateScript [
   &ParticleState()
   &SpriteRotationRate()
   &UpdateMeshOrientation(Rotation Vector=&RandomRangeVector(Minimum=<-0.2 -0.2 -0.2>, Maximum=<1 1 1>))
   &ScaleSpriteSize(Uniform Curve Sprite Scale=DataInterface() {propertyValues {Curve {keys [
        {tangentMode=RCTM_User, time=0, value=1}
        {arriveTangentWeight=0.2822, interpMode=RCIM_Cubic, leaveTangentWeight=0.1893, tangentMode=RCTM_User, time=0.5985, value=1}
        {arriveTangent=-2.7277, interpMode=RCIM_Cubic, leaveTangent=-2.7277, tangentMode=RCTM_User, time=1, value=0}
       ]}}})
   &ScaleColor(
    Scale Alpha=&FloatFromCurve(FloatCurve=DataInterface() {propertyValues {Curve {keys [<at -0.0046 1.0113 RCIM_Cubic>,<at 1.0011 0.002 RCIM_Cubic>]}}})
    Scale RGB=&MakeVectorFromLinearColorRGB(Color=&FNTimeOfDayColor(
      Night Color=<rgba 0.4102 0.5995 0.75 1>
      Morning Color=<rgba 1 0.6765 0.6094 1>
      Day Color=<rgba 0.7865 0.7865 0.7865 1>
      Evening Color=<rgba 0.7708 0.7708 0.7708 1>
     ))
    ScaleA=true
    ScaleRGB=true
   )
   &Drag(
    Drag=&RandomRangeFloat(Minimum=5, Maximum=7)
    Rotational Drag=2
    Use Linear Drag=true
    Use Rotational Drag=true
   )
   &SolveForcesAndVelocity(Clamp Velocity=false, Limit Acceleration=false)
   &AlignSpriteToMeshOrientation(SetAlignment=true, SetFacing=true)
  ]
  rendererClasses [&NiagaraSpriteRendererProperties]
  renderers [{rendererClass=&NiagaraSpriteRendererProperties}]
 }
 RootsIntro {
  emitterSpawnScript [
   &InitializeAudioPlayer(Audio Player=DataInterface() {propertyValues {
      Attenuation=null
      Concurrency=null
      MaxPlaysPerTick=10
      SoundToPlay=/CosmeticCompanions/Assets/Audio_Shared/Interaction/Petting/SC_Mimosa_Interaction_React_Default
      bAllowLoopingOneShotSounds=false
      bLimitPlaysPerTick=true
      bOnlyActiveDuringGameplay=false
      bStopWhenComponentIsDestroyed=true
      ConfigurationUserParameter {parameter {name=None, typeDefHandle {registeredTypeIndex=1}, varData []}}
      ParameterNames []
     }})
   /Niagara/Modules/Audio/PlayAudio() {module=PlayAudio001}
  ]
  emitterUpdateScript [
   &EmitterState()
   &SpawnBurst_Instantaneous(Use Spawn Probability=false, UseLoopCountLimit=false)
  ]
  particleSpawnScript [
   &InitializeParticle(
    Color=&MakeCustomLinearColorFromBool(
     Bool=$User.bIsFrontEndPreview
     True Color=<rgba 0.5 0.5 0.5 1>
     False Color=&FNTimeOfDayColor(
      Night Color=<rgba 0.5366 0.5587 0.625 1>
      Morning Color=<rgba 0.7188 0.649 0.577 1>
      Day Color=<rgba 0.8385 0.8385 0.8385 1>
      Evening Color=<rgba 0.7188 0.649 0.577 1>
     )
    )
    Position Offset=$System.BurstPosition
    Position Offset Coordinate Space=::Local
    Sprite Size Mode=::NewEnumerator3
    Uniform Sprite Size=28
    UsePositionOffset=true
    Write Material Random=true
   )
   /SetVariables(Particles.RightRootSize:Vector2f=<35 35>) {isSetParameters=true}
   &CameraOffset(Camera Offset Amount=10)
  ]
  particleUpdateScript [
   &ParticleState()
   &SubUVAnimation(
    "Sprite Renderer "=DataInterface() {propertyValues {SpriteRenderer=&self:RootsIntro_0.NiagaraSpriteRendererProperties_0}}
    SubUV Lookup Index:NiagaraFloat=&FloatFromCurve(FloatCurve=DataInterface() {propertyValues {Curve {keys [<at 0 0>,<at 0.35 1>]}}})
    UseEndFrame=false
    UseStartFrame=false
   )
   &OffsetPosition(Position Offset=&Lerp_Vector(A=$System.HeartIntroSpawnPosition, B=$System.BurstPosition, Alpha=0.97)) {enabled=false}
   &ScaleColor(
    Scale Alpha=&FloatFromCurve(FloatCurve=DataInterface() {propertyValues {Curve {keys [<at 0 1>,<at 0.8938 1>,<at 1 0>]}}})
    ScaleA=true
    ScaleRGB=true
   )
  ]
  rendererClasses [&NiagaraSpriteRendererProperties]
  renderers [
   {rendererClass=&NiagaraSpriteRendererProperties}
   {rendererClass=&NiagaraSpriteRendererProperties, rendererIndex=1}
  ]
 }
}
summary {
 systemName=NS_ReactFX_DroneGravy
 emitters [
  {emitterName=RootsIntro, rendererClasses [&NiagaraSpriteRendererProperties]}
  {emitterName=BigHeart, rendererClasses []}
  {emitterName=Leaf_Out, rendererClasses [&NiagaraSpriteRendererProperties]}
  {emitterName=HeartGlow, rendererClasses []}
  {emitterName=LWE_Rays, rendererClasses []}
  {emitterName=LWE_Corona, rendererClasses []}
 ]
 userVariables [
  User.bIsFrontEnd {type=NiagaraBool}
  User.bIsFrontEndPreview {type=NiagaraBool, defaultValue=true}
 ]
}
