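# Niagara reaction effect for the EffigyPin cosmetic companion: an animated sprite travels to a burst point and pops, with the shared petting react sound.
# Intro lerps from HeartIntroSpawnPosition to BurstPosition over IntroLength, trailed by the wavy InkTrail ribbon.
# As it ends, Pop bursts indexed pieces along hand-authored offsets driven by RippingCurveAnim, plus a glow.
# Front-end modifiers adjust size and offset; Aura, Ring, Impact, Shards002 and Skull_Spline are disabled.
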NiagaraSystem /CosmeticCompanions/FX/React/NS_ReactFX_EffigyPin
dependencies {
 usedDataInterfaces [
  &NiagaraDataInterfaceAudioPlayer
  &NiagaraDataInterfaceCurve
  &NiagaraDataInterfaceSpriteRendererInfo
  &NiagaraDataInterfaceVectorCurve
  &NiagaraDataInterfaceColorCurve
  &NiagaraDataInterfaceMeshRendererInfo
 ]
 usedDynamicInputs [&RandomRangeInt, /Niagara/DynamicInputs/TypeConversions/MakeCustomPositionFromBool]
 usedModules [
  &self:SystemSpawnScript.SystemScriptSource.SystemScriptGraph.NiagaraNodeAssignment_0.SetVariables_#1
  &self:SystemSpawnScript.SystemScriptSource.SystemScriptGraph.NiagaraNodeAssignment_1.SetVariables_#2
  &self:SystemSpawnScript.SystemScriptSource.SystemScriptGraph.NiagaraNodeAssignment_6.SetVariables_#3
  &self:SystemSpawnScript.SystemScriptSource.SystemScriptGraph.NiagaraNodeAssignment_4.SetVariables_#4
  &self:SystemSpawnScript.SystemScriptSource.SystemScriptGraph.NiagaraNodeAssignment_7.SetVariables_#5
  &self:SystemSpawnScript.SystemScriptSource.SystemScriptGraph.NiagaraNodeAssignment_5.SetVariables_#6
  &self:SystemSpawnScript.SystemScriptSource.SystemScriptGraph.NiagaraNodeAssignment_2.SetVariables_#7
  &self:SystemSpawnScript.SystemScriptSource.SystemScriptGraph.NiagaraNodeAssignment_3.SetVariables_#8
  &SystemState
  &self:SystemSpawnScript.SystemScriptSource.SystemScriptGraph.NiagaraNodeAssignment_8.SetVariables_#9
 ]
}
emitters {
 Aura {
  bEnabled=false
  emitterUpdateScript [
   &EmitterState()
   &SpawnRate(
    SpawnRate=&FloatFromCurve(
     FloatCurve=DataInterface() {propertyValues {Curve {keys [<at 0 0>,<at 0.1 0>,<at 0.16 1>,<at 0.25 1>,<at 0.275 0>]}}}
     CurveIndex=$Emitter.NormalizedLoopAge
     Scale Curve=18
    )
    Use Spawn Probability=false
   )
  ]
  particleSpawnScript [
   &InitializeParticle(
    Color=&Multiply_LinearColorByFloat(
     Linear Color=<rgba 0.776 0.7655 0.1981 1>
     Float=&FNTimeOfDayScalar(Night Scale=0.5, Morning Scale=0.78, Evening Scale=0.78)
    )
    Lifetime=0.25
    Position Offset=$System.BurstPosition
    Sprite Rotation Mode=::NewEnumerator2
    Sprite Size Mode=::NewEnumerator3
    Uniform Sprite Size=&FN_FrontEndFloatModifier(
     Gameplay=&RandomRangeFloat(Minimum=45, Maximum=55)
     Front End=&RandomRangeFloat(Minimum=35, Maximum=45)
    )
    UsePositionOffset=true
    Write Material Random=true
   )
   &CameraOffset(Camera Offset Amount=&RandomRangeFloat(Minimum=-15, Maximum=-10))
   /SetVariables(
    Particles.Magic Large Color:LinearColor=<rgba 0.8646 0.8145 0.0495 1>
    Particles.Magic Large Size:Vector2f=&Vector2DFromFloat(Value=&FN_FrontEndFloatModifier(
      Gameplay=&RandomRangeFloat(Minimum=65, Maximum=80)
      Front End=&RandomRangeFloat(Minimum=55, Maximum=70)
     ))
   ) {isSetParameters=true}
  ]
  particleUpdateScript [
   &ParticleState()
   &ScaleSpriteSize(Uniform Curve Sprite Scale=DataInterface() {propertyValues {Curve {keys [
        {arriveTangentWeight=0.0807, interpMode=RCIM_Cubic, leaveTangentWeight=0.3949, tangentMode=RCTM_User, time=0, value=0}
        {arriveTangent=1.1137, interpMode=RCIM_Cubic, leaveTangent=1.1137, time=0.15, value=1}
        {arriveTangentWeight=0.3423, interpMode=RCIM_Cubic, leaveTangentWeight=0.0493, tangentMode=RCTM_User, time=0.8979, value=1}
        <at 1 0 RCIM_Cubic>
       ]}}})
   &SpriteRotationRate(Rotation Rate=&RandomRangeFloat(Minimum=-100, Maximum=100))
   &SolveForcesAndVelocity(Clamp Velocity=false, Limit Acceleration=false)
   /SetVariables(
    Particles.Scale Magic Large Color:LinearColor=&MakeLinearColorFromVectorAndFloat(
     "Vector (RGB)"=&MakeVectorFromLinearColorRGB(Color=$Particles.Initial.Magic Large Color)
     "Float (Alpha)"=&FloatFromCurve(FloatCurve=DataInterface() {propertyValues {Curve {keys [<at 0 1>,<at 1 1>]}}})
    )
    Particles.Scale Magic Large Size:Vector2f=&Vector2DFromFloat(Value=&Multiply_Float(
      A=&MakeFloatFromVector2d(Vector 2D=$Particles.Initial.Magic Large Size)
      B=&FloatFromCurve(FloatCurve=DataInterface() {propertyValues {Curve {keys [
           {arriveTangentWeight=0.0807, interpMode=RCIM_Cubic, leaveTangentWeight=0.3949, tangentMode=RCTM_User, time=0, value=0}
           {arriveTangent=1.1137, interpMode=RCIM_Cubic, leaveTangent=1.1137, time=0.3, value=1}
           {arriveTangentWeight=0.3423, interpMode=RCIM_Cubic, leaveTangentWeight=0.0493, tangentMode=RCTM_User, time=0.8979, value=1}
           <at 1 0 RCIM_Cubic>
          ]}}})
     ))
   ) {isSetParameters=true}
   &ScaleColor(
    Scale Alpha=&FloatFromCurve(FloatCurve=DataInterface() {propertyValues {Curve {
        defaultValue=1
        keys [
         {interpMode=RCIM_Cubic, leaveTangentWeight=0.4986, tangentMode=RCTM_User, time=0, value=0}
         {arriveTangentWeight=0.0742, interpMode=RCIM_Cubic, leaveTangentWeight=0.2988, tangentMode=RCTM_User, time=0.1, value=1}
         {interpMode=RCIM_Cubic, tangentMode=RCTM_User, time=0.5026, value=1}
         {arriveTangentWeight=0.1658, interpMode=RCIM_Cubic, tangentMode=RCTM_User, time=0.9145, value=0}
        ]
       }}})
    ScaleA=true
    ScaleRGB=true
   ) {enabled=false}
  ]
  rendererClasses [&NiagaraSpriteRendererProperties]
  renderers [
   {rendererClass=&NiagaraSpriteRendererProperties}
   {rendererClass=&NiagaraSpriteRendererProperties, rendererIndex=1}
  ]
 }
 Impact {bEnabled=false, rendererClasses [], renderers []}
 InkTrail {
  emitterSpawnScript [/SetVariables(Emitter.RandomOffset:NiagaraFloat=&RandomRangeFloat(Minimum=0, Maximum=30)) {isSetParameters=true}]
  emitterUpdateScript [
   &EmitterState()
   &SpawnRate(
    SpawnGroup=1
    SpawnRate=&FloatFromCurve(
     FloatCurve=DataInterface() {propertyValues {Curve {keys [<at 0 1 RCIM_Constant>,<at 0.3 0 RCIM_Constant>]}}}
     CurveIndex=$Emitter.NormalizedLoopAge
     Scale Curve=30
    )
    Use Spawn Probability=false
   )
  ]
  particleSpawnScript [
   &InitializeParticle(
    Lifetime=0.5
    Ribbon Width=&Multiply_Float(A=8, B=&FN_FrontEndFloatModifier(Gameplay=1, Front End=$User.FrontEndScale))
    Ribbon Width Mode=::NewEnumerator1
    Sprite Rotation Mode=::NewEnumerator0
    UsePositionOffset=false
    Write Material Random=true
   )
   /SetVariables(Particles.Position=&Add_Vector(
     A=<0 0 6>
     B=&Lerp_Vector(
      A=$System.HeartIntroSpawnPosition
      B=$System.BurstPosition
      Alpha=&FloatFromCurve(
       FloatCurve=DataInterface() {propertyValues {Curve {keys [{arriveTangent=2.5323, interpMode=RCIM_Cubic, leaveTangent=2.5323, tangentMode=RCTM_User, time=0, value=0}, <at 1 1 RCIM_Cubic>]}}}
       CurveIndex=&Divide_Float(A=$Emitter.Age, B=$System.IntroLength)
      )
     )
    )) {isSetParameters=true}
   &AddVelocity(
    Velocity=&MakeVector(
     X=&Waveform(
      Global Input Position=$Emitter.RandomOffset
      Global Phase=&Multiply_FloatByInt(Float=0.01, Integer=&ReturnParticleIDAttributes())
      "Waveform [1]"=::NewEnumerator0
      "[1] Frequency"=12
      "[1] Amplitude Min/Max"=<-1 1>
      "[1] Scale"=6
     )
     Y=&MakeFloatFromVector(
      Vector=&MakeVector(
       X=&Waveform(
        Global Input Position=$Emitter.RandomOffset
        Global Phase=&Multiply_FloatByInt(Float=0.01, Integer=&ReturnParticleIDAttributes())
        "Waveform [1]"=::NewEnumerator1
        "[1] Frequency"=12
        "[1] Amplitude Min/Max"=<-1 1>
        "[1] Scale"=6
       )
       Y=0
       Z=0
      )
      Channel=::NewEnumerator0
     )
     Z=&RandomRangeFloat(Minimum=0, Maximum=-5)
    )
    Velocity Speed Scale=&FloatFromCurve(
     FloatCurve=DataInterface() {propertyValues {Curve {keys [<at 0.5 1>,<at 1 0>]}}}
     CurveIndex=&Divide_Float(A=$Emitter.Age, B=$System.IntroLength)
     Scale Curve=&RandomRangeFloat(Minimum=3, Maximum=5)
    )
   )
   /SetVariables(Particles.RibbonUVDistance:NiagaraFloat=&DistanceBetweenPositions(
     A=&Lerp_Vector(
      A=$System.HeartIntroSpawnPosition
      B=$System.BurstPosition
      Alpha=&FloatFromCurve(
       FloatCurve=DataInterface() {propertyValues {Curve {keys [
           {arriveTangent=0.7913, interpMode=RCIM_Cubic, leaveTangent=0.7913, tangentMode=RCTM_User, time=0, value=0}
           {arriveTangent=-0.1073, tangentMode=RCTM_User, time=1, value=1}
          ]}}}
       CurveIndex=&Divide_Float(A=$Emitter.Age, B=$System.IntroLength)
      )
     )
     B=<0 0 0>
    )) {isSetParameters=true}
   /SetVariables(Particles.SpawnAge:NiagaraFloat=&Divide_Float(A=$Emitter.Age, B=$System.IntroLength)) {isSetParameters=true}
  ]
  particleUpdateScript [
   &ParticleState()
   &ScaleColor(
    Scale Alpha=&Multiply_Float(
     A=&FloatFromCurve(FloatCurve=DataInterface() {propertyValues {Curve {
         defaultValue=0
         keys [
          {arriveTangent=-0.2441, time=0.5, value=1}
          {arriveTangent=-0.0405, interpMode=RCIM_Cubic, leaveTangent=-0.0405, tangentMode=RCTM_User, time=1, value=0}
         ]
        }}})
     B=0.7
    )
    Scale RGB=&MakeVectorFromLinearColorRGB(Color=&ColorFromCurve(ColorCurve=DataInterface() {propertyValues {
        AlphaCurve {keys [<at 0 1>]}
        BlueCurve {keys [<at 0 0.6384>,<at 1 1>]}
        GreenCurve {keys [<at 0 0.7188>,<at 1 0.3698>]}
        RedCurve {keys [<at 0 0.643>,<at 1 0.715>]}
       }}))
    ScaleA=true
    ScaleRGB=true
   )
   &DynamicMaterialParameters(Module.Index0Float0Write=true, Module.Index0Float1Write=true, Module.Index0Float2Write=true, Module.Index0Float3Write=true) {enabled=false}
   &DynamicMaterialParameters(
    Index 0 Param 1=-0.5
    Index 0 Param 3=1
    Index 0 Param 4=1
    Module.Index0Float0Write=true
    Module.Index0Float1Write=true
    Module.Index0Float2Write=true
    Module.Index0Float3Write=true
   ) {module=DynamicMaterialParameters001}
   &ScaleRibbonWidth(Ribbon Width Scale=&Multiply_Float(
     A=&FloatFromCurve(FloatCurve=DataInterface() {propertyValues {Curve {keys [<at 0 1>,<at 0.3 1>]}}})
     B=&FloatFromCurve(FloatCurve=DataInterface() {propertyValues {Curve {keys [<at 0.05 0>,<at 0.15 1>,<at 0.5 1>,<at 0.9 0>]}}}, CurveIndex=$Particles.SpawnAge)
    ))
   &Drag(Drag=2, Use Linear Drag=true, Use Rotational Drag=true)
   &SolveForcesAndVelocity(Clamp Velocity=false, Limit Acceleration=false)
  ]
  rendererClasses [&NiagaraRibbonRendererProperties]
  renderers [{rendererClass=&NiagaraRibbonRendererProperties}]
 }
 Intro {
  emitterSpawnScript [
   &InitializeAudioPlayer(Audio Player=DataInterface() {propertyValues {
      Attenuation=null
      Concurrency=null
      MaxPlaysPerTick=10
      SoundToPlay=/CosmeticCompanions/Assets/Audio_Shared/Interaction/Petting/SC_Mimosa_Interaction_React_Default
      bAllowLoopingOneShotSounds=false
      bLimitPlaysPerTick=true
      bOnlyActiveDuringGameplay=false
      bStopWhenComponentIsDestroyed=true
      ConfigurationUserParameter {parameter {name=None, typeDefHandle {registeredTypeIndex=1}, varData []}}
      ParameterNames []
     }})
   /Niagara/Modules/Audio/PlayAudio() {module=PlayAudio001}
  ]
  emitterUpdateScript [
   &EmitterState()
   &SpawnBurst_Instantaneous(Use Spawn Probability=false, UseLoopCountLimit=false)
  ]
  particleSpawnScript [&InitializeParticle(
    Lifetime=$System.IntroLength
    Ribbon Width=10
    Ribbon Width Mode=::NewEnumerator1
    Sprite Rotation Mode=::NewEnumerator0
    Sprite Size Mode=::NewEnumerator3
    Uniform Sprite Size=&Multiply_Float(A=30, B=&FN_FrontEndFloatModifier(Gameplay=1, Front End=$User.FrontEndScale))
    UsePositionOffset=false
    Write Material Random=true
   )]
  particleUpdateScript [
   &ParticleState()
   &ScaleSpriteSize(
    Scale Sprite Size Mode=::NewEnumerator0
    Uniform Scale Factor=&Add_Float(
     A=0
     B=&FloatFromCurve(FloatCurve=DataInterface() {propertyValues {Curve {
         defaultValue=0
         keys [
          <at 0 0.3 RCIM_Cubic>
          {arriveTangent=1.25, leaveTangent=1.25, time=0.4, value=1}
          {arriveTangent=6.566, interpMode=RCIM_Cubic, leaveTangent=6.566, tangentMode=RCTM_User, time=0.9, value=1.1}
         ]
        }}})
    )
   ) {module=ScaleSpriteSize002}
   &ScaleColor(
    Scale Alpha=&FloatFromCurve(FloatCurve=DataInterface() {propertyValues {Curve {
        defaultValue=0
        keys [
         {arriveTangent=9.0725, interpMode=RCIM_Cubic, leaveTangent=9.0725, tangentMode=RCTM_User, time=0, value=0}
         {arriveTangent=-0.2441, leaveTangent=-0.2441, time=0.1408, value=0.9847}
         {arriveTangent=-1.146, leaveTangent=-1.146, time=0.95, value=1}
         {arriveTangent=-0.0405, interpMode=RCIM_Cubic, leaveTangent=-0.0405, tangentMode=RCTM_User, time=1, value=0}
        ]
       }}})
    ScaleA=true
    ScaleRGB=true
   )
   &OffsetPosition(Position Offset=&Lerp_Vector(
     A=$System.HeartIntroSpawnPosition
     B=$System.BurstPosition
     Alpha=&FloatFromCurve(FloatCurve=DataInterface() {propertyValues {Curve {keys [
          {arriveTangent=2.2068, interpMode=RCIM_Cubic, leaveTangent=2.2068, tangentMode=RCTM_User, time=0, value=0}
          {arriveTangent=-0.1074, tangentMode=RCTM_User, time=0.7, value=1}
         ]}}})
    ))
   &DynamicMaterialParameters(
    Index 0 Param 2=1
    Module.Index0Float0Write=true
    Module.Index0Float1Write=true
    Module.Index0Float2Write=true
    Module.Index0Float3Write=true
   )
   &SubUVAnimation(End Frame Range Override=8, Output Alternate Renderer Binding=true, UseEndFrame=true, UseStartFrame=true)
  ]
  rendererClasses [&NiagaraSpriteRendererProperties]
  renderers [
   {rendererClass=&NiagaraSpriteRendererProperties}
   {rendererClass=&NiagaraSpriteRendererProperties, rendererIndex=1}
  ]
 }
 Pop {
  emitterUpdateScript [
   &EmitterState()
   &SpawnBurst_Instantaneous(
    Spawn Count=7
    Spawn Time=&Subtract_Float(A=$System.IntroLength, B=0.125)
    Use Spawn Probability=false
    UseLoopCountLimit=false
   )
   &SpawnBurst_Instantaneous(
    Spawn Group=1
    Spawn Time=&Subtract_Float(A=$System.IntroLength, B=0.1)
    Use Spawn Probability=false
    UseLoopCountLimit=false
   ) {module=SpawnBurst_Instantaneous001}
  ]
  particleSpawnScript [
   /SetVariables(Particles.VisibilityTag=&SpawnGroup()) {isSetParameters=true}
   /SetVariables(Particles.SubImageIndex=&MakeFloatFromInt(Int=&Add_Integer(A=&ReturnExecIndex(), B=0))) {isSetParameters=true}
   &InitializeParticle(
    Lifetime=&MaskFloatBySpawnGroup(Masked Float=&RandomRangeFloat(Minimum=0.125, Maximum=0.625), Passthrough Float=0.5)
    Position Offset=$System.BurstPosition
    Sprite Rotation Mode=::NewEnumerator0
    Sprite Size Mode=::NewEnumerator3
    Uniform Sprite Size=&Multiply_Float(
     A=&MergeHelper_FloatSwitch(0=1.25, 1=3.2, 2=2, 3=3, 4=3.4, 5=2.8, 6=3, Index=&ReturnExecIndex(), Num=::NewEnumerator5)
     B=&FN_FrontEndFloatModifier(Gameplay=1, Front End=$User.FrontEndScale)
    )
    UsePositionOffset=true
    Write Material Random=true
   )
  ]
  particleUpdateScript [
   &ParticleState()
   &SpriteRotationRate(Rotation Rate=&FloatFromCurve(
     FloatCurve=DataInterface() {propertyValues {Curve {
        defaultValue=1
        keys [
         {arriveTangent=2.6847, interpMode=RCIM_Cubic, leaveTangent=2.6847, tangentMode=RCTM_User, time=0, value=0}
         {arriveTangent=0.1978, interpMode=RCIM_Cubic, leaveTangent=0.1978, tangentMode=RCTM_User, time=1.0073, value=1.0194}
        ]
       }}}
     Scale Curve=&UniformAOrBFloat(A=-750, B=750, Evaluation Type=::NewEnumerator0, Recalculate Random Each Loop=true, Override Seed=false)
    ))
   &ScaleSpriteSize(
    Scale Sprite Size Mode=::NewEnumerator0
    Uniform Scale Factor=&FloatFromCurve(FloatCurve=DataInterface() {propertyValues {Curve {keys [
         {arriveTangent=3.3458, interpMode=RCIM_Cubic, leaveTangent=3.3458, tangentMode=RCTM_User, time=0, value=1}
         {arriveTangent=-1, interpMode=RCIM_Cubic, leaveTangent=-1, time=0.5, value=3}
         {arriveTangent=0.0352, interpMode=RCIM_Cubic, leaveTangent=0.0352, tangentMode=RCTM_User, time=1, value=0}
        ]}}})
   ) {module=ScaleSpriteSize002}
   &OffsetPosition(Position Offset=&AddVectorToPosition(
     Position=Linked() {linkedVariable {name=System.BurstPosition, type=Vector3f}}
     Vector=&AddVectorToPosition(
      Position=&FN_FrontEndVector3Modifier(Gameplay=<0 0 0>, Front End=<0 0 12>)
      Vector=&MergeHelper_VectorSwitch(
       0=&Multiply_VectorByFloat(
        Vector=&MakeVector(
         X=0
         Y=0
         Z=&FloatFromCurve(FloatCurve=DataInterface() {propertyValues {Curve {keys [<at 0.5 30 RCIM_Cubic>,<at 1 10 RCIM_Cubic>]}}})
        )
        Float=&FloatFromCurve(FloatCurve=$System.RippingCurveAnim)
       )
       1=&Multiply_VectorByFloat(
        Vector=&MakeVector(
         X=-15
         Y=-5
         Z=&FloatFromCurve(FloatCurve=DataInterface() {propertyValues {Curve {keys [<at 0.5 20 RCIM_Cubic>,<at 1 0 RCIM_Cubic>]}}})
        )
        Float=&FloatFromCurve(FloatCurve=$System.RippingCurveAnim, Scale Curve=1.2)
       )
       2=&Multiply_VectorByFloat(
        Vector=&MakeVector(
         X=-30
         Y=-7
         Z=&FloatFromCurve(FloatCurve=DataInterface() {propertyValues {Curve {keys [<at 0.5 5 RCIM_Cubic>,<at 1 -25 RCIM_Cubic>]}}})
        )
        Float=&FloatFromCurve(FloatCurve=$System.RippingCurveAnim, Scale Curve=0.8)
       )
       3=&Multiply_VectorByFloat(
        Vector=&MakeVector(
         X=-10
         Y=-9
         Z=&FloatFromCurve(FloatCurve=DataInterface() {propertyValues {Curve {keys [<at 0.5 -15 RCIM_Cubic>,<at 1 -65 RCIM_Cubic>]}}})
        )
        Float=&FloatFromCurve(FloatCurve=$System.RippingCurveAnim)
       )
       4=&Multiply_VectorByFloat(
        Vector=&MakeVector(
         X=10
         Y=6
         Z=&FloatFromCurve(FloatCurve=DataInterface() {propertyValues {Curve {keys [<at 0.5 -9 RCIM_Cubic>,<at 1 -24 RCIM_Cubic>]}}})
        )
        Float=&FloatFromCurve(FloatCurve=$System.RippingCurveAnim, Scale Curve=0.941)
       )
       5=&Multiply_VectorByFloat(
        Vector=&MakeVector(
         X=19
         Y=4
         Z=&FloatFromCurve(FloatCurve=DataInterface() {propertyValues {Curve {keys [<at 0.5 7 RCIM_Cubic>,<at 1 -12 RCIM_Cubic>]}}})
        )
        Float=&FloatFromCurve(FloatCurve=$System.RippingCurveAnim, Scale Curve=1.2)
       )
       6=&Multiply_VectorByFloat(
        Vector=&MakeVector(
         X=20
         Y=-2
         Z=&FloatFromCurve(FloatCurve=DataInterface() {propertyValues {Curve {keys [<at 0.5 35 RCIM_Cubic>,<at 1 12 RCIM_Cubic>]}}})
        )
        Float=&FloatFromCurve(FloatCurve=$System.RippingCurveAnim)
       )
       Index=&ReturnExecIndex()
       Num=::NewEnumerator5
      )
     )
    )) {module=OffsetPosition001}
   /SetVariables(
    Particles.Glow Color=&Multiply_LinearColorByFloat(
     Linear Color=&MultiplyLinearColor(
      Linear Color=$Particles.Color
      Scale Factor=&MakeVector4_FromLinearColor(LinearColor=<rgba 1 0.8281 0.1504 1>)
     )
     Float=&FloatFromCurve(
      FloatCurve=DataInterface() {propertyValues {Curve {keys [
          <at 0 1>
          {arriveTangent=-5.1023, interpMode=RCIM_Cubic, leaveTangent=-5.1023, tangentMode=RCTM_User, time=0.5, value=1}
          <at 0.8 0 RCIM_Cubic>
         ]}}}
      Scale Curve=0.33
     )
    )
    Particles.GlowPosition:NiagaraPosition=&AddVectorToPosition(Position=Linked() {linkedVariable {name=System.BurstPosition, type=Vector3f}}, Vector=<0 0 0>)
    Particles.GlowSize=&Vector2DFromFloat(Value=&FloatFromCurve(
      FloatCurve=DataInterface() {propertyValues {Curve {keys [{arriveTangent=3.9275, interpMode=RCIM_Cubic, leaveTangent=3.9275, tangentMode=RCTM_User, time=0, value=1}, <at 1 3 RCIM_Cubic>]}}}
      Scale Curve=20
     ))
   ) {isSetParameters=true}
   &SubUVAnimation(
    End Frame Range Override=16
    Evaluation Type=::NewEnumerator0
    "Sprite Renderer "=DataInterface() {propertyValues {SpriteRenderer=&self:NiagaraEmitter_2.NiagaraSpriteRendererProperties_2}}
    Start Frame Range Override=10
    SubUV Animation Mode:ENiagara_SubUVLookupModeV2=::NewEnumerator2
    UseEndFrame=true
    UseStartFrame=true
   )
  ]
  rendererClasses [&NiagaraSpriteRendererProperties]
  renderers [
   {rendererClass=&NiagaraSpriteRendererProperties}
   {rendererClass=&NiagaraSpriteRendererProperties, rendererIndex=1}
  ]
 }
 Ring {
  bEnabled=false
  emitterUpdateScript [
   &EmitterState()
   &SpawnBurst_Instantaneous(Spawn Time=0.08, Use Spawn Probability=false, UseLoopCountLimit=false)
  ]
  particleSpawnScript [
   &InitializeParticle(
    Color=<rgba 3 3 3 1>
    Lifetime=0.725
    Mesh Scale=&Multiply_VectorByFloat(Vector=<2.3354 1.45 1.45>, Float=0.25)
    Mesh Scale Mode=::NewEnumerator4
    Sprite Size Mode=::NewEnumerator3
    Uniform Sprite Size=50
    UsePositionOffset=false
    Write Material Random=true
   )
   /SetVariables(Particles.GlowSpriteFacing:Vector3f=<0 0 1>) {isSetParameters=true}
   &InitialMeshOrientation(Mesh Orientation Mode=::NewEnumerator1, Rotation=<0 0.25 0>, Use Rotation Vector=true)
   &AddVelocity(Velocity=<0 0 -1>)
  ]
  particleUpdateScript [
   &ParticleState()
   &ScaleMeshSize(Scale Factor=&VectorFromCurve(VectorCurve=DataInterface() {propertyValues {XCurve {keys [<at 0 0.4>,<at 1 1>]}, YCurve {keys [<at 0 0.4>,<at 1 1>]}, ZCurve {keys [<at 0 0.4>,<at 1 1>]}}})) {module=ScaleMeshSize001}
   &SolveForcesAndVelocity(Clamp Velocity=false, Limit Acceleration=false)
   &ScaleColor(
    Linear Color Curve=DataInterface() {propertyValues {
      AlphaCurve {keys [
        {arriveTangent=9.0725, interpMode=RCIM_Cubic, leaveTangent=9.0725, tangentMode=RCTM_User, time=0, value=0}
        {
         arriveTangent=-3.699
         arriveTangentWeight=0.0469
         interpMode=RCIM_Cubic
         leaveTangent=-3.699
         tangentMode=RCTM_User
         time=0.1408
         value=0.9847
        }
        {arriveTangent=-0.0405, arriveTangentWeight=0.2864, interpMode=RCIM_Cubic, leaveTangent=-0.0405, tangentMode=RCTM_User, time=1, value=0}
       ]}
      BlueCurve {keys [<at 0 0.1227>,<at 1 0.1451>]}
      GreenCurve {keys [<at 0 0.7537>,<at 1 0.4894>]}
      RedCurve {keys [<at 0 1>,<at 1 1>]}
     }}
    Scale Mode=::NewEnumerator2
   )
   /SetVariables(
    Particles.WindColor:LinearColor=&ColorFromCurve(ColorCurve=DataInterface() {propertyValues {
       AlphaCurve {keys [
         {arriveTangent=9.0725, interpMode=RCIM_Cubic, leaveTangent=9.0725, tangentMode=RCTM_User, time=0, value=0}
         {arriveTangent=-1.6136, arriveTangentWeight=0.0469, interpMode=RCIM_Cubic, leaveTangent=-1.6136, tangentMode=RCTM_User, time=0.141, value=0.3}
         {arriveTangent=-0.0405, arriveTangentWeight=0.2864, interpMode=RCIM_Cubic, leaveTangent=-0.0405, tangentMode=RCTM_User, time=1, value=0}
        ]}
       BlueCurve {keys [<at 0 0.1882 RCIM_Cubic>,<at 0.9969 0>]}
       GreenCurve {keys [<at 0 0.702 RCIM_Cubic>,<at 0.9969 0>]}
       RedCurve {keys [<at 0 1 RCIM_Cubic>,<at 0.9969 0>]}
      }})
    Particles.WindDP:Vector4f=<0.5 0.6 0 0>
   ) {isSetParameters=true}
   &OffsetPosition(Position Offset=&Lerp_Vector(
     A=$System.HeartIntroSpawnPosition
     B=&AddVectorToPosition(Position=Linked() {linkedVariable {name=System.BurstPosition, type=Vector3f}}, Vector=<0 0 0>)
     Alpha=&FloatFromCurve(FloatCurve=DataInterface() {propertyValues {Curve {keys [
          {arriveTangent=3.0606, interpMode=RCIM_Cubic, leaveTangent=3.0606, tangentMode=RCTM_User, time=0, value=0}
          {arriveTangent=-0.1073, tangentMode=RCTM_User, time=1, value=1}
         ]}}})
    ))
  ]
  rendererClasses [&NiagaraMeshRendererProperties]
  renderers [{rendererClass=&NiagaraMeshRendererProperties, rendererIndex=0}]
 }
 Shards002 {bEnabled=false, rendererClasses [], renderers []}
 Skull_Spline {
  bEnabled=false
  emitterUpdateScript [
   &EmitterState(Life Cycle Mode=::NewEnumerator1, Loop Behavior=::NewEnumerator1, Loop Duration=1.5, UseLoopDelay=false)
   &SpawnBurst_Instantaneous(Spawn Time=0.03, Use Spawn Probability=false, UseLoopCountLimit=false)
  ]
  particleSpawnScript [
   &InitializeParticle(
    Lifetime=1.7
    Mesh Scale=&Multiply_VectorByFloat(Vector=<0.2399 0.182 0.96>, Float=0.6707)
    Mesh Scale Mode=::NewEnumerator4
    UsePositionOffset=false
    Write Material Random=true
   )
   &AddVelocity(
    Velocity=&RotateVectorByQuaternion(Vector To Rotate=<0 1 1>, Quaternion=$Particles.SkeletalMeshLocation.SampledRotation)
    Velocity Speed Scale=&RandomRangeFloat(Minimum=10, Maximum=20)
   )
  ]
  particleUpdateScript [
   &ParticleState()
   &ScaleMeshSize(Scale Factor=&MakeVector(
     X=&FN_FrontEndFloatModifier(Gameplay=1, Front End=0.6)
     Y=1
     Z=&FloatFromCurve(
      FloatCurve=DataInterface() {propertyValues {Curve {keys [
          {arriveTangent=9.0725, interpMode=RCIM_Cubic, leaveTangent=9.0725, tangentMode=RCTM_User, time=0, value=0}
          {arriveTangent=0.7067, interpMode=RCIM_Cubic, leaveTangent=0.7067, tangentMode=RCTM_User, time=0.017, value=0.974}
          {arriveTangent=0.5062, interpMode=RCIM_Cubic, leaveTangent=0.5062, time=0.7704, value=1.1212}
          {arriveTangent=0.6578, interpMode=RCIM_Cubic, leaveTangent=0.6578, tangentMode=RCTM_User, time=1, value=1.4716}
         ]}}}
      Scale Curve=&FN_FrontEndFloatModifier(Gameplay=0.85, Front End=0.7)
     )
    ))
   &CurlNoiseForce(
    Noise Strength=&FloatFromCurve(
     FloatCurve=DataInterface() {propertyValues {Curve {keys [{interpMode=RCIM_Cubic, tangentMode=RCTM_User, time=0, value=0}, {interpMode=RCIM_Cubic, tangentMode=RCTM_User, time=0.7, value=1}]}}}
     Scale Curve=&RandomRangeFloat(Minimum=25, Maximum=50)
    )
    Pan Noise=false
   )
   &Drag(Drag=0.5, Use Linear Drag=true, Use Rotational Drag=true)
   &SolveForcesAndVelocity(Clamp Velocity=false, Limit Acceleration=false)
   &DynamicMaterialParameters(
    Index 0 Param 1=&FloatFromCurve(FloatCurve=DataInterface() {propertyValues {Curve {keys [
         {arriveTangent=5.0399, interpMode=RCIM_Cubic, leaveTangent=5.0399, tangentMode=RCTM_User, time=0, value=-1}
         {arriveTangent=0.9837, interpMode=RCIM_Cubic, leaveTangent=0.9837, tangentMode=RCTM_User, time=0.1997, value=-0.2096}
         {arriveTangent=0.0502, interpMode=RCIM_Cubic, leaveTangent=0.0502, tangentMode=RCTM_User, time=0.6, value=-0.0272}
        ]}}})
    Index 0 Param 2=&FloatFromCurve(
     FloatCurve=DataInterface() {propertyValues {Curve {keys [
         {arriveTangent=0.0771, interpMode=RCIM_Cubic, leaveTangent=0.0771, tangentMode=RCTM_User, time=0.5768, value=0}
         {arriveTangent=2.1363, interpMode=RCIM_Cubic, leaveTangent=2.1363, tangentMode=RCTM_User, time=1, value=0.75}
        ]}}}
     Scale Curve=&RandomRangeFloat(Minimum=0.4, Maximum=0.6)
    )
    Index 0 Param 3=&FloatFromCurve(FloatCurve=DataInterface() {propertyValues {Curve {keys [
         {arriveTangent=-0.4297, interpMode=RCIM_Cubic, leaveTangent=-0.4297, tangentMode=RCTM_User, time=0, value=0.4549}
         {arriveTangent=-0.0708, interpMode=RCIM_Cubic, leaveTangent=-0.0708, tangentMode=RCTM_User, time=0.1638, value=0}
         {arriveTangent=0.0092, interpMode=RCIM_Cubic, leaveTangent=0.0092, tangentMode=RCTM_User, time=0.6109, value=0}
         {arriveTangent=3.7616, interpMode=RCIM_Cubic, leaveTangent=3.7616, tangentMode=RCTM_User, time=1, value=1}
        ]}}})
    Index 0 Param 4=&RandomRangeFloat(Minimum=0, Maximum=1)
    Index 1 Param 1=&FloatFromCurve(
     FloatCurve=DataInterface() {propertyValues {Curve {keys [
         {arriveTangent=0.0825, interpMode=RCIM_Cubic, leaveTangent=0.0825, tangentMode=RCTM_User, time=0.7079, value=0}
         {arriveTangent=2.7761, interpMode=RCIM_Cubic, leaveTangent=2.7761, tangentMode=RCTM_User, time=1, value=1}
        ]}}}
     Scale Curve=&RandomRangeFloat(Minimum=0.8, Maximum=1)
    )
    Index 1 Param 2=&FloatFromCurve(
     FloatCurve=DataInterface() {propertyValues {Curve {keys [
         {arriveTangent=-0.0388, interpMode=RCIM_Cubic, leaveTangent=-0.0388, tangentMode=RCTM_User, time=0, value=0}
         {arriveTangent=0.1591, interpMode=RCIM_Cubic, leaveTangent=0.1591, tangentMode=RCTM_User, time=0.3, value=1}
         {interpMode=RCIM_Cubic, tangentMode=RCTM_User, time=0.45, value=1}
         {arriveTangent=2.0344, interpMode=RCIM_Cubic, leaveTangent=2.0344, tangentMode=RCTM_User, time=1, value=2}
        ]}}}
     Scale Curve=0.6444
    )
    Index 2 Param 1=&FloatFromCurve(FloatCurve=DataInterface() {propertyValues {Curve {keys [
         {arriveTangent=0.0429, interpMode=RCIM_Cubic, leaveTangent=0.0429, tangentMode=RCTM_User, time=0.5026, value=0.3926}
         {interpMode=RCIM_Cubic, tangentMode=RCTM_User, time=1, value=0.8848}
        ]}}})
    Index 2 Param 2=&MakeFloatFromVector(Vector=$Particles.SkeletalMeshLocation.SampledPosition, Channel=::NewEnumerator1)
    Index 2 Param 3=&FloatFromCurve(
     FloatCurve=DataInterface() {propertyValues {Curve {keys [
         {interpMode=RCIM_Cubic, tangentMode=RCTM_User, time=0, value=1}
         {arriveTangent=-0.0336, interpMode=RCIM_Cubic, leaveTangent=-0.0336, tangentMode=RCTM_User, time=0.5, value=1}
         {interpMode=RCIM_Cubic, tangentMode=RCTM_User, time=1, value=0}
        ]}}}
     Scale Curve=&RandomRangeFloat(Minimum=0.8, Maximum=1.1)
    )
    Index 2 Param 4=&FloatFromCurve(FloatCurve=DataInterface() {propertyValues {Curve {keys [
         {interpMode=RCIM_Cubic, tangentMode=RCTM_User, time=0, value=1}
         {arriveTangent=0.6667, interpMode=RCIM_Cubic, leaveTangent=0.6667, time=0.2, value=0.55}
         <at 0.3 1.2 RCIM_Cubic>
         {arriveTangent=-0.0265, interpMode=RCIM_Cubic, leaveTangent=-0.0265, tangentMode=RCTM_User, time=0.4, value=0.55}
         {arriveTangent=0.0518, interpMode=RCIM_Cubic, leaveTangent=0.0518, tangentMode=RCTM_User, time=0.5, value=0.9}
         {interpMode=RCIM_Cubic, tangentMode=RCTM_User, time=0.8, value=0}
        ]}}})
    Module.Index0Float0Write=true
    Module.Index0Float1Write=true
    Module.Index0Float2Write=true
    Module.Index0Float3Write=true
    Module.Index1Float0Write=true
    Module.Index1Float1Write=true
    Module.Index1Float2Write=true
    Module.Index1Float3Write=true
    Module.Index2Float0Write=true
    Module.Index2Float1Write=true
    Module.Index2Float2Write=true
    Module.Index2Float3Write=true
    Write Parameter Index 1=::NewEnumerator1
    Write Parameter Index 2=::NewEnumerator1
   )
   &ScaleColor(
    Scale Alpha=&FloatFromCurve(
     FloatCurve=DataInterface() {propertyValues {Curve {keys [
         {interpMode=RCIM_Cubic, tangentMode=RCTM_User, time=0.8, value=1}
         {arriveTangent=-6.2584, interpMode=RCIM_Cubic, leaveTangent=-6.2584, tangentMode=RCTM_User, time=1, value=0}
        ]}}}
     Scale Curve=0.2
    )
    Scale RGB=&VectorFromFloat(Value=&FloatFromCurve(FloatCurve=DataInterface() {propertyValues {Curve {keys [
          {arriveTangent=19.0414, interpMode=RCIM_Cubic, leaveTangent=19.0414, tangentMode=RCTM_User, time=0, value=2.095}
          {interpMode=RCIM_Cubic, tangentMode=RCTM_User, time=0.3031, value=4}
          {arriveTangent=0.0291, interpMode=RCIM_Cubic, leaveTangent=0.0291, tangentMode=RCTM_User, time=0.6, value=4}
          {arriveTangent=-0.0639, interpMode=RCIM_Cubic, leaveTangent=-0.0639, tangentMode=RCTM_User, time=0.8, value=1}
         ]}}}))
    ScaleA=true
    ScaleRGB=true
   )
   &ScaleColor(
    Scale RGB=&MakeVectorFromLinearColorRGB(Color=&FNTimeOfDayColor(
      Night Color=&ColorFromCurve(ColorCurve=DataInterface() {propertyValues {
         AlphaCurve {keys []}
         BlueCurve {keys [<at 0 1.6007 RCIM_Cubic>,<at 0.134 5 RCIM_Cubic>]}
         GreenCurve {keys [<at 0 0.0809 RCIM_Cubic>,<at 0.134 0.7269 RCIM_Cubic>]}
         RedCurve {keys [<at 0 0.3147 RCIM_Cubic>,<at 0.134 2.3841 RCIM_Cubic>]}
        }})
      Morning Color=&ColorFromCurve(ColorCurve=DataInterface() {propertyValues {
         AlphaCurve {keys []}
         BlueCurve {keys [<at 0 1.6007 RCIM_Cubic>,<at 0.134 7 RCIM_Cubic>]}
         GreenCurve {keys [<at 0 0.0824 RCIM_Cubic>,<at 0.134 1.0901 RCIM_Cubic>]}
         RedCurve {keys [<at 0 0.537 RCIM_Cubic>,<at 0.134 5.6664 RCIM_Cubic>]}
        }})
      Day Color=&ColorFromCurve(ColorCurve=DataInterface() {propertyValues {
         AlphaCurve {keys []}
         BlueCurve {keys [<at 0 1.6007 RCIM_Cubic>,<at 0.134 10 RCIM_Cubic>]}
         GreenCurve {keys [<at 0 0.0824 RCIM_Cubic>,<at 0.134 1.5573 RCIM_Cubic>]}
         RedCurve {keys [<at 0 0.537 RCIM_Cubic>,<at 0.134 8.0948 RCIM_Cubic>]}
        }})
      Evening Color=&ColorFromCurve(ColorCurve=DataInterface() {propertyValues {
         AlphaCurve {keys []}
         BlueCurve {keys [<at 0 1.6007 RCIM_Cubic>,<at 0.134 7 RCIM_Cubic>]}
         GreenCurve {keys [<at 0 0.0824 RCIM_Cubic>,<at 0.134 1.0901 RCIM_Cubic>]}
         RedCurve {keys [<at 0 0.537 RCIM_Cubic>,<at 0.134 5.6664 RCIM_Cubic>]}
        }})
     ))
    ScaleA=true
    ScaleRGB=true
   ) {module=ScaleColor001}
  ]
  rendererClasses [&NiagaraMeshRendererProperties]
  renderers [{rendererClass=&NiagaraMeshRendererProperties, rendererIndex=0}]
 }
}
summary {
 systemName=NS_ReactFX_EffigyPin
 emitters [
  {emitterName=Intro, rendererClasses [&NiagaraSpriteRendererProperties]}
  {emitterName=Pop, rendererClasses [&NiagaraSpriteRendererProperties]}
  {bEnabled=false, emitterName=Impact, rendererClasses []}
  {bEnabled=false, emitterName=Ring, rendererClasses [&NiagaraMeshRendererProperties]}
  {bEnabled=false, emitterName=Shards002, rendererClasses []}
  {bEnabled=false, emitterName=Aura, rendererClasses [&NiagaraSpriteRendererProperties]}
  {emitterName=InkTrail, rendererClasses [&NiagaraRibbonRendererProperties]}
  {bEnabled=false, emitterName=Skull_Spline, rendererClasses [&NiagaraMeshRendererProperties]}
 ]
 userVariables [
  User.bIsFrontEnd {type=NiagaraBool}
  User.bIsFrontEndPreview {type=NiagaraBool}
  User.FrontEndScale {type=NiagaraFloat, defaultValue=0.6, description=""}
 ]
}
