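# Niagara system for a cosmetic companion reaction effect (FacetWhisk): a staged sprite intro that ends in a purple-tinted burst of shards and smoke.
# Intro002 plays a shared petting-reaction sound and spawns three staggered sprite groups, using per-group switches for lifetime, color and dynamic material parameters.
# One second in, Ice_Frags005 throws shrinking mesh fragments with sprites, alongside SubUV smoke with time-of-day alpha.
# Front-end modifiers adjust sizes; the LWE emitters list no renderers.
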NiagaraSystem /CosmeticCompanions/FX/React/NS_ReactFX_FacetWhisk
dependencies {
 usedDataInterfaces [
  &NiagaraDataInterfaceMeshRendererInfo
  &NiagaraDataInterfaceColorCurve
  &NiagaraDataInterfaceCurve
  &NiagaraDataInterfaceVector2DCurve
  &NiagaraDataInterfaceAudioPlayer
 ]
 usedDynamicInputs [
  &Multiply_Vector2D_ByFloat
  /Niagara/DynamicInputs/Add/Add_Vector2
  &Vector2DFromCurve
  &FN_FrontEndVector3Modifier
  &AddVectorToPosition
  &Add_Float
  &Vector2DFromFloat
  &Multiply_Float
  &MakeCustomLinearColorFromBool
  &Boolean_OrOperation
 ]
 usedModules [
  &self:SystemSpawnScript.SystemScriptSource.SystemScriptGraph.NiagaraNodeAssignment_0.SetVariables_#1
  &self:SystemSpawnScript.SystemScriptSource.SystemScriptGraph.NiagaraNodeAssignment_1.SetVariables_#2
  &self:SystemSpawnScript.SystemScriptSource.SystemScriptGraph.NiagaraNodeAssignment_4.SetVariables_#3
  &self:SystemSpawnScript.SystemScriptSource.SystemScriptGraph.NiagaraNodeAssignment_3.SetVariables_#4
  &self:SystemSpawnScript.SystemScriptSource.SystemScriptGraph.NiagaraNodeAssignment_2.SetVariables_#5
  &SystemState
 ]
}
emitters {
 Ice_Frags005 {
  emitterUpdateScript [
   &EmitterState()
   &SpawnBurst_Instantaneous(Spawn Count=15, Spawn Time=1, Use Spawn Probability=false, UseLoopCountLimit=false)
  ]
  particleSpawnScript [
   &InitializeParticle(
    Color=<rgba 2.5 2.5 2.5 1>
    Lifetime Max=0.5
    Lifetime Min=0.23
    Lifetime Mode=::NewEnumerator1
    Mesh Renderer Array Visibility Mode=::NewEnumerator1
    Mesh Renderer Info=DataInterface() {propertyValues {MeshRenderer=&self:Ice_Frags005_1.NiagaraMeshRendererProperties_2}}
    Mesh Renderer Mesh Index:NiagaraInt32=&RandomRangeInt(Minimum=0, Maximum=1)
    Mesh Scale Max=<0.15 0.15 0.2>
    Mesh Scale Min=<0.075 0.075 0.1>
    Mesh Scale Mode=::NewEnumerator2
    Sprite Size=<5 10>
    Sprite Size Mode=::NewEnumerator4
    UsePositionOffset=false
    Write Material Random=true
   )
   &CameraOffset(Camera Offset Amount=&RandomRangeFloat(Minimum=0, Maximum=20, Evaluation Type=::NewEnumerator1))
   &ShapeLocation(MaskSpawnGroup=false, Override Seed=false, Shape Origin:Vector3f=$System.BurstPosition, Sphere Radius=4, UseOwnerScale=true)
   &AddVelocity(
    Module.Use Falloff Distance=false
    Override Seed=false
    Velocity Mode=::NewEnumerator1
    Velocity Origin=$Particles.ShapeLocation.ShapePosition
    Velocity Speed=&RandomRangeFloat(Minimum=50, Maximum=100)
   )
  ]
  particleUpdateScript [
   &ParticleState()
   &SolveForcesAndVelocity(Clamp Velocity=false, Limit Acceleration=false)
   &ScaleMeshSize(Scale Factor=&VectorFromFloat(Value=&FloatFromCurve(FloatCurve=DataInterface() {propertyValues {Curve {
         defaultValue=0.6697
         keys [
          {tangentMode=RCTM_User, time=0, value=1}
          {arriveTangent=-0.0405, interpMode=RCIM_Cubic, leaveTangent=-0.0405, tangentMode=RCTM_User, time=0.7009, value=1}
          {arriveTangent=-8.0089, interpMode=RCIM_Cubic, leaveTangent=-8.0089, tangentMode=RCTM_User, time=1, value=0}
         ]
        }}})))
   &ScaleColor(
    Scale Alpha=&FloatFromCurve(FloatCurve=DataInterface() {propertyValues {Curve {keys [<at 0.5 1 RCIM_Cubic>,<at 1.0011 0.002 RCIM_Cubic>]}}})
    ScaleA=true
    ScaleRGB=true
   )
   &FN_Sprite_Attributes(
    Color=&MakeLinearColorFromVectorAndFloat(
     "Vector (RGB)"=&MakeVectorFromLinearColorRGB(Color=<rgba 0.669 0.0856 1 1>)
     "Float (Alpha)"=&FloatFromCurve(
      FloatCurve=DataInterface() {propertyValues {Curve {
         defaultValue=0.6697
         keys [
          {tangentMode=RCTM_User, time=0, value=1}
          {arriveTangent=-0.0405, interpMode=RCIM_Cubic, leaveTangent=-0.0405, tangentMode=RCTM_User, time=0.7009, value=1}
          {arriveTangent=-8.0089, interpMode=RCIM_Cubic, leaveTangent=-8.0089, tangentMode=RCTM_User, time=1, value=0}
         ]
        }}}
      Scale Curve=0.05
     )
    )
    Create Color Attribute=true
    Create SpriteSize Attribute=true
    Module Output:ENiagara_AttributeSamplingApplyOutput=::NewEnumerator1
    SpriteSize=<20 20>
   )
  ]
  rendererClasses [&NiagaraMeshRendererProperties, &NiagaraSpriteRendererProperties]
  renderers [
   {rendererClass=&NiagaraMeshRendererProperties, rendererIndex=0}
   {rendererClass=&NiagaraSpriteRendererProperties, rendererIndex=1}
  ]
 }
 Intro002 {
  emitterSpawnScript [
   &InitializeAudioPlayer(Audio Player=DataInterface() {propertyValues {
      Attenuation=null
      Concurrency=null
      MaxPlaysPerTick=10
      SoundToPlay=/CosmeticCompanions/Assets/Audio_Shared/Interaction/Petting/SC_Mimosa_Interaction_React_Long
      bAllowLoopingOneShotSounds=false
      bLimitPlaysPerTick=true
      bOnlyActiveDuringGameplay=false
      bStopWhenComponentIsDestroyed=true
      ConfigurationUserParameter {parameter {name=None, typeDefHandle {registeredTypeIndex=1}, varData []}}
      ParameterNames []
     }})
   /Niagara/Modules/Audio/PlayAudio() {module=PlayAudio001}
   /SetVariables(Emitter.RotationRandom:NiagaraFloat=&MakeCustomFloatFromBool(Bool=&RandomBool(), True Float=1, False Float=-1)) {isSetParameters=true}
  ]
  emitterUpdateScript [
   &EmitterState(Life Cycle Mode=::NewEnumerator1, Loop Behavior=::NewEnumerator1, Loop Duration=1, UseLoopDelay=false)
   &SpawnBurst_Instantaneous(Use Spawn Probability=false, UseLoopCountLimit=false)
   &SpawnBurst_Instantaneous(Spawn Group=1, Spawn Time=0.25, Use Spawn Probability=false, UseLoopCountLimit=false) {module=SpawnBurst_Instantaneous002}
   &SpawnBurst_Instantaneous(Spawn Group=2, Spawn Time=0.7, Use Spawn Probability=false, UseLoopCountLimit=false) {module=SpawnBurst_Instantaneous003}
   &SpawnBurst_Instantaneous(Spawn Group=1, Use Spawn Probability=false, UseLoopCountLimit=false) {enabled=false, module=SpawnBurst_Instantaneous001}
  ]
  particleSpawnScript [
   /SetVariables(Particles.VisibilityTag=&SpawnGroup()) {isSetParameters=true}
   &InitializeParticle(
    Color=&MergeHelper_ColorSwitch(0=<rgba 9.6244 2.37 10 1>, 1=<rgba 1 1 1 1>, 2=<rgba 1 1 1 1>, 3=?, Index=$Particles.VisibilityTag, Num=::NewEnumerator1)
    Lifetime=&MergeHelper_FloatSwitch(0=1.2, 1=1.2, 2=0.9, 3=?, Index=$Emitter.SpawnGroup, Num=::NewEnumerator1)
    Position Offset=$System.BurstPosition
    Sprite Rotation Mode=::NewEnumerator0
    Sprite Size Mode=::NewEnumerator3
    Uniform Sprite Size=&FN_FrontEndFloatModifier(Gameplay=25, Front End=19.5)
    UsePositionOffset=true
    Write Material Random=true
   )
  ]
  particleUpdateScript [
   &ParticleState()
   &ScaleSpriteSize(
    Scale Sprite Size Mode=::NewEnumerator0
    Uniform Scale Factor=&FloatFromCurve(
     FloatCurve=DataInterface() {propertyValues {Curve {keys [
         {arriveTangent=1.2333, interpMode=RCIM_Cubic, leaveTangent=1.2333, tangentMode=RCTM_User, time=0, value=0.2}
         {arriveTangent=5.6726, interpMode=RCIM_Cubic, leaveTangent=5.6726, time=0.0414, value=0.2829}
         {arriveTangent=4.8098, interpMode=RCIM_Cubic, leaveTangent=4.8098, time=0.0947, value=0.7374}
         {arriveTangent=0.1086, interpMode=RCIM_Cubic, leaveTangent=0.1086, tangentMode=RCTM_User, time=0.1905, value=1}
         {arriveTangent=0.8399, interpMode=RCIM_Cubic, leaveTangent=0.8399, tangentMode=RCTM_User, time=0.4866, value=1.1757}
         {arriveTangent=0.7883, interpMode=RCIM_Cubic, leaveTangent=0.7883, time=0.6798, value=1.2763}
         {arriveTangent=0.1199, interpMode=RCIM_Constant, tangentMode=RCTM_User, time=0.8075, value=1.4287}
         <at 0.8233 1.6809>
         <at 0.9046 1.3243 RCIM_Constant>
         <at 0.9151 2.0277>
         <at 0.9989 0.6118 RCIM_Constant>
         {arriveTangent=-4.0496, interpMode=RCIM_Cubic, leaveTangent=-4.0496, time=1.0073, value=0.4151}
         {arriveTangent=-1.5827, interpMode=RCIM_Cubic, leaveTangent=-1.5827, tangentMode=RCTM_User, time=1.15, value=0}
        ]}}}
     CurveIndex=$Particles.Age
     Scale Curve=0.75
    )
   ) {enabled=false}
   &DynamicMaterialParameters(
    Index 0 Param 1=&MergeHelper_FloatSwitch(
     0=&FloatFromCurve(FloatCurve=DataInterface() {propertyValues {Curve {defaultValue=1, keys [{tangentMode=RCTM_User, time=0, value=1}, {tangentMode=RCTM_User, time=0.4, value=0}]}}})
     1=&FloatFromCurve(FloatCurve=DataInterface() {propertyValues {Curve {defaultValue=1, keys [{tangentMode=RCTM_User, time=0, value=1}, {tangentMode=RCTM_User, time=0.4, value=0}]}}})
     2=&FloatFromCurve(FloatCurve=DataInterface() {propertyValues {Curve {
         defaultValue=1
         keys [{arriveTangent=-3.0206, interpMode=RCIM_Cubic, leaveTangent=-3.0206, tangentMode=RCTM_User, time=0, value=1}, <at 0.5 0 RCIM_Cubic>]
        }}})
     3=?
     Index=$Particles.VisibilityTag
     Num=::NewEnumerator1
    )
    Index 0 Param 2=&MergeHelper_FloatSwitch(0=5, 1=0.02, 2=0.1, 3=?, Index=$Particles.VisibilityTag, Num=::NewEnumerator1)
    Module.Index0Float0Write=true
    Module.Index0Float1Write=true
    Module.Index0Float2Write=true
    Module.Index0Float3Write=true
   )
   &ScaleColor(
    Scale RGB=&MergeHelper_VectorSwitch(
     0=&VectorFromFloat(Value=&FloatFromCurve(
       FloatCurve=DataInterface() {propertyValues {Curve {keys [{arriveTangent=-24.9931, interpMode=RCIM_Cubic, leaveTangent=-24.9931, tangentMode=RCTM_User, time=0, value=10}, <at 1 1 RCIM_Cubic>]}}}
       Scale Curve=1.5
      ))
     1=&VectorFromFloat(Value=&FloatFromCurve(
       FloatCurve=DataInterface() {propertyValues {Curve {keys [{arriveTangent=-24.9931, interpMode=RCIM_Cubic, leaveTangent=-24.9931, tangentMode=RCTM_User, time=0, value=10}, <at 0.3 1 RCIM_Cubic>]}}}
       Scale Curve=1.5
      ))
     2=&VectorFromFloat(Value=&FloatFromCurve(FloatCurve=DataInterface() {propertyValues {Curve {keys [<at 0.3 1 RCIM_Constant>,<at 0.3839 3 RCIM_Constant>,<at 0.45 1 RCIM_Constant>]}}}))
     3=?
     Index=$Particles.VisibilityTag
     Num=::NewEnumerator1
    )
    ScaleA=true
    ScaleRGB=true
   )
   &ScaleSpriteSize(
    Uniform Curve Index=$Emitter.Age
    Uniform Curve Sprite Scale=DataInterface() {propertyValues {Curve {
       defaultValue=0
       keys [
        {tangentMode=RCTM_User, time=0, value=1}
        {arriveTangent=-0.0405, interpMode=RCIM_Cubic, leaveTangent=-0.0405, tangentMode=RCTM_User, time=1, value=1}
        {arriveTangent=-5, interpMode=RCIM_Cubic, leaveTangent=-5, time=1.15, value=1.2}
        {arriveTangent=-17.7139, interpMode=RCIM_Cubic, leaveTangent=-17.7139, tangentMode=RCTM_User, time=1.2, value=0}
       ]
      }}}
   ) {module=ScaleSpriteSize001}
   /SetVariables(
    Particles.Glow Color=&MakeLinearColorFromVectorAndFloat(
     "Vector (RGB)"=&MakeVectorFromLinearColorRGB(Color=<rgba 0.2736 0.0662 0.429 1>)
     "Float (Alpha)"=&FloatFromCurve(
      FloatCurve=DataInterface() {propertyValues {Curve {
         defaultValue=1
         keys [
          <at 0.317 0 RCIM_Cubic>
          {arriveTangent=2.3189, interpMode=RCIM_Cubic, leaveTangent=2.3189, time=0.6291, value=0.3321}
          <at 0.7482 1 RCIM_Cubic>
         ]
        }}}
      Scale Curve=0.1
     )
    )
    Particles.Glow Size=&Multiply_Vector2D(A=$Particles.SpriteSize, B=<2.5 2.5>)
   ) {isSetParameters=true}
  ]
  rendererClasses [&NiagaraSpriteRendererProperties]
  renderers [
   {rendererClass=&NiagaraSpriteRendererProperties}
   {rendererClass=&NiagaraSpriteRendererProperties, rendererIndex=1}
   {rendererClass=&NiagaraSpriteRendererProperties, rendererIndex=2}
   {rendererClass=&NiagaraSpriteRendererProperties, rendererIndex=3}
  ]
 }
 LWE_AirGlow {rendererClasses [], renderers []}
 LWE_Corona {rendererClasses [], renderers []}
 LWE_Rays {rendererClasses [], renderers []}
 Smoke {
  emitterUpdateScript [
   &EmitterState(Scalability Mode=::NewEnumerator1)
   &SpawnBurst_Instantaneous(Spawn Count=6, Spawn Time=1, Use Spawn Probability=false, UseLoopCountLimit=false)
  ]
  particleSpawnScript [
   &InitializeParticle(
    Color=<rgba 0.1 0.1 0.1 1>
    Lifetime Max=0.8
    Lifetime Min=0.4
    Lifetime Mode=::NewEnumerator1
    Sprite Rotation Mode=::NewEnumerator2
    Sprite Size Mode=::NewEnumerator1
    Uniform Sprite Size Max=50
    Uniform Sprite Size Min=40
    UsePositionOffset=false
    Write Material Random=true
   ) {module=InitializeParticle002}
   /SetVariables(Particles.Rotation:NiagaraFloat=&RandomRangeFloat(Minimum=-80, Maximum=80)) {isSetParameters=true}
   /SetVariables(Particles.RandomFrame=&RandomRangeInt(Minimum=0, Maximum=63)) {isSetParameters=true}
   &ShapeLocation(MaskSpawnGroup=false, Offset=<0 0 2>, Offset Mode=::NewEnumerator0, Override Seed=false, Sphere Radius=10, UseOwnerScale=true)
  ]
  particleUpdateScript [
   &ParticleState()
   &SolveForcesAndVelocity(Clamp Velocity=false, Limit Acceleration=false)
   &ScaleColor(
    Scale Alpha=&FloatFromCurve(
     FloatCurve=DataInterface() {propertyValues {Curve {keys [<at 0 0 RCIM_Cubic>,<at 0.4 0.9932 RCIM_Cubic>,<at 1 0>]}}}
     Scale Curve=&FNTimeOfDayScalar(Night Scale=0.3, Morning Scale=0.1, Evening Scale=0.1)
    )
    Scale RGB=&MakeVectorFromLinearColorRGB(Color=&ColorFromCurve(ColorCurve=DataInterface() {propertyValues {
        AlphaCurve {keys [<at 0.1151 1>,<at 0.9977 1 RCIM_Cubic>]}
        BlueCurve {keys [<at 0.1176 1>,<at 0.3853 1>,<at 0.7334 1>,<at 0.9868 0.5469>]}
        GreenCurve {keys [<at 0.1176 0.0349>,<at 0.3853 0.0671>,<at 0.7334 0>,<at 0.9868 0>]}
        RedCurve {keys [<at 0.1176 0.2965>,<at 0.3853 0.6332>,<at 0.7334 0.0453>,<at 0.9868 0.0593>]}
       }}))
    ScaleA=true
    ScaleRGB=true
   ) {module=ScaleColor001}
   &ScaleSpriteSize(
    Uniform Curve Scale=&FN_FrontEndFloatModifier(Gameplay=1, Front End=$System.FrontendScale)
    Uniform Curve Sprite Scale=DataInterface() {propertyValues {Curve {keys [
        {arriveTangent=0.46, interpMode=RCIM_Cubic, leaveTangent=0.46, tangentMode=RCTM_User, time=0, value=1}
        {arriveTangent=0.2144, interpMode=RCIM_Cubic, leaveTangent=0.2144, tangentMode=RCTM_User, time=1, value=1.2}
       ]}}}
   ) {module=ScaleSpriteSize001}
   &FN_SubUVAnimation(End Frame=&Add_Integer(A=$Particles.RandomFrame, B=63), Start Frame=$Particles.RandomFrame, SubUV Loop Count=0.3)
  ]
  rendererClasses [&NiagaraSpriteRendererProperties]
  renderers [{rendererClass=&NiagaraSpriteRendererProperties}]
 }
}
summary {
 systemName=NS_ReactFX_FacetWhisk
 emitters [
  {emitterName=Smoke, rendererClasses [&NiagaraSpriteRendererProperties]}
  {emitterName=LWE_AirGlow, rendererClasses []}
  {emitterName=LWE_Corona, rendererClasses []}
  {emitterName=LWE_Rays, rendererClasses []}
  {emitterName=Ice_Frags005, rendererClasses [&NiagaraMeshRendererProperties, &NiagaraSpriteRendererProperties]}
  {emitterName=Intro002, rendererClasses [&NiagaraSpriteRendererProperties]}
 ]
 userVariables [
  User.bIsFrontEnd {type=NiagaraBool}
  User.bIsFrontEndPreview {type=NiagaraBool}
  User.FrontEndScale {type=NiagaraFloat, defaultValue=0.7, description=""}
 ]
}
