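# Sound cue for the FlourCat companion's petting reaction, using the default Mimosa petting reaction wave.
# A fixed half-second delay feeds a single wave player, so the sound starts shortly after triggering.
# It uses companion foley attenuation and the creature pets sound class, with slightly reduced volume and its own concurrency override limiting simultaneous instances.
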SoundCue /CosmeticCompanions/Assets/Audio_Shared/Interaction/Petting/SC_Mimosa_Interaction_React_FlourCat
root=Delay_0
nodes [
 {id=Delay_0, children [WavePlayer_6], properties {delayMax=0.5, delayMin=0.5}}
 {
  id=WavePlayer_6
  properties {soundWaveAssetPtr=/CosmeticCompanions/Assets/Audio_Shared/Interaction/Petting/Wavs/React_Default/Mimosa_Petting_React_Default}
 }
]
properties {
 attenuationSettings=/Game/Sounds/Attenuation/AS_Companion_Foley
 bOverrideConcurrency=true
 maxDistance=1200
 soundClassObject=&Fort_Creatures_Pets
 volumeMultiplier=0.7
 concurrencyOverrides {maxCount=4}
}
