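# Niagara reaction effect for a cosmetic companion (FlourCut) that plays a petting-interaction sound and a heart burst.
# An Intro sprite eases from the spawn position to a burst position, trailed by a wavy ribbon.
# Near arrival, a two-part Pop fires and five hearts spread toward random system-level offsets with anim-trail ribbons.
# Front-end modifiers adjust size and glow, and time-of-day color tints the intro glow and pop dust.
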NiagaraSystem /CosmeticCompanions/FX/React/NS_ReactFX_FlourCut
dependencies {
 usedDataInterfaces [
  &NiagaraDataInterfaceAudioPlayer
  &NiagaraDataInterfaceCurve
  &NiagaraDataInterfaceVectorCurve
  &NiagaraDataInterfaceColorCurve
 ]
 usedDynamicInputs [
  /Niagara/DynamicInputs/TypeConversions/MakeCustomPositionFromBool
  &FN_FrontEndVector3Modifier
  &Multiply_VectorByFloat
  &RandomRangeVector
 ]
 usedModules [
  &self:SystemSpawnScript.SystemScriptSource.SystemScriptGraph.NiagaraNodeAssignment_0.SetVariables_#1
  &self:SystemSpawnScript.SystemScriptSource.SystemScriptGraph.NiagaraNodeAssignment_1.SetVariables_#2
  &self:SystemSpawnScript.SystemScriptSource.SystemScriptGraph.NiagaraNodeAssignment_3.SetVariables_#3
  &self:SystemSpawnScript.SystemScriptSource.SystemScriptGraph.NiagaraNodeAssignment_4.SetVariables_#4
  &self:SystemSpawnScript.SystemScriptSource.SystemScriptGraph.NiagaraNodeAssignment_7.SetVariables_#5
  &self:SystemSpawnScript.SystemScriptSource.SystemScriptGraph.NiagaraNodeAssignment_5.SetVariables_#6
  &self:SystemSpawnScript.SystemScriptSource.SystemScriptGraph.NiagaraNodeAssignment_2.SetVariables_#7
  &SystemState
  &self:SystemSpawnScript.SystemScriptSource.SystemScriptGraph.NiagaraNodeAssignment_6.SetVariables_#8
 ]
}
emitters {
 Hearts {
  emitterUpdateScript [
   &EmitterState()
   &SpawnBurst_Instantaneous(Spawn Count=5, Spawn Time=0.56, Use Spawn Probability=false, UseLoopCountLimit=false)
  ]
  particleSpawnScript [
   &FN_SetSpawnGroup(Number of Spawn Groups=5)
   &InitializeParticle(
    Lifetime=&MergeHelper_FloatSwitch(0=0.6, 1=0.65, 2=0.69, 3=0.72, 4=0.77, Index=&SpawnGroup(), Num=::NewEnumerator3)
    Sprite Rotation Mode=::NewEnumerator0
    Sprite Size Mode=::NewEnumerator3
    Uniform Sprite Size=13
    UsePositionOffset=false
    Write Material Random=true
   )
  ]
  particleUpdateScript [
   &ParticleState()
   &ScaleSpriteSize(
    Scale Sprite Size Mode=::NewEnumerator0
    Uniform Scale Factor=&Add_Float(
     A=0.5
     B=&FloatFromCurve(FloatCurve=DataInterface() {propertyValues {Curve {
         defaultValue=0
         keys [
          {arriveTangent=3.6909, interpMode=RCIM_Cubic, leaveTangent=3.6909, tangentMode=RCTM_User, time=0, value=0}
          {arriveTangent=2.055, interpMode=RCIM_Cubic, leaveTangent=2.055, tangentMode=RCTM_User, time=0.7247, value=0.9545}
          {arriveTangent=-2.4453, interpMode=RCIM_Cubic, leaveTangent=-2.4453, time=0.9148, value=1.7368}
          {arriveTangent=-0.1202, interpMode=RCIM_Cubic, leaveTangent=-0.1202, tangentMode=RCTM_User, time=1, value=0.2814}
         ]
        }}})
    )
   ) {module=ScaleSpriteSize002}
   &ScaleColor(
    Scale Alpha=&FloatFromCurve(FloatCurve=DataInterface() {propertyValues {Curve {
        defaultValue=0
        keys [
         {arriveTangent=9.0725, interpMode=RCIM_Cubic, leaveTangent=9.0725, tangentMode=RCTM_User, time=0, value=0}
         {arriveTangent=-0.2441, leaveTangent=-0.2441, time=0.1408, value=0.9847}
         {arriveTangent=-1.146, leaveTangent=-1.146, time=0.95, value=1}
         {arriveTangent=-0.0405, interpMode=RCIM_Cubic, leaveTangent=-0.0405, tangentMode=RCTM_User, time=1, value=0}
        ]
       }}})
    ScaleA=true
    ScaleRGB=true
   )
   &OffsetPosition(Position Offset=&Lerp_Vector(
     A=&Add_Vector(A=$System.BurstPosition, B=<0 0 3>)
     B=&Add_Vector(
      A=$System.BurstPosition
      B=&Multiply_Vector(
       A=<1 1 1>
       B=&MergeHelper_VectorSwitch(
        0=$System.RandomVectorPos1
        1=$System.RandomVectorPos2
        2=$System.RandomVectorPos3
        3=$System.RandomVectorPos4
        4=$System.RandomVectorPos5
        Index=&SpawnGroup()
        Num=::NewEnumerator3
       )
      )
     )
     Alpha=&FloatFromCurve(FloatCurve=DataInterface() {propertyValues {Curve {keys [
          {arriveTangent=0.7913, interpMode=RCIM_Cubic, leaveTangent=0.7913, tangentMode=RCTM_User, time=0, value=0}
          {arriveTangent=-0.1074, tangentMode=RCTM_User, time=1, value=1}
         ]}}})
    ))
   &DynamicMaterialParameters(
    Index 0 Param 1=&FloatFromCurve(
     FloatCurve=DataInterface() {propertyValues {Curve {keys [{arriveTangent=0.6296, interpMode=RCIM_Cubic, leaveTangent=0.6296, tangentMode=RCTM_User, time=0, value=0.8771}, <at 0.75 0.9676>, <at 1 1>]}}}
     Scale Curve=0.8
    )
    Module.Index0Float0Write=true
    Module.Index0Float1Write=true
    Module.Index0Float2Write=true
    Module.Index0Float3Write=true
   )
   /SetVariables(
    Particles.Glow Color=&Multiply_LinearColorByFloat(Linear Color=$System.Layer 2 Color, Float=0.1)
    Particles.Glow Size=&Multiply_Vector2D(A=$Particles.SpriteSize, B=<0.6 0.6>)
   ) {isSetParameters=true, module=SetVariables_#9}
  ]
  rendererClasses [&NiagaraSpriteRendererProperties]
  renderers [
   {rendererClass=&NiagaraSpriteRendererProperties}
   {rendererClass=&NiagaraSpriteRendererProperties, rendererIndex=1}
  ]
 }
 Intro {
  emitterSpawnScript [
   &InitializeAudioPlayer(Audio Player=DataInterface() {propertyValues {
      Attenuation=null
      Concurrency=null
      MaxPlaysPerTick=10
      SoundToPlay=/CosmeticCompanions/Assets/Audio_Shared/Interaction/Petting/SC_Mimosa_Interaction_React_FlourCat
      bAllowLoopingOneShotSounds=false
      bLimitPlaysPerTick=true
      bOnlyActiveDuringGameplay=false
      bStopWhenComponentIsDestroyed=true
      ConfigurationUserParameter {parameter {name=None, typeDefHandle {registeredTypeIndex=1}, varData []}}
      ParameterNames []
     }})
   /Niagara/Modules/Audio/PlayAudio() {module=PlayAudio001}
  ]
  emitterUpdateScript [
   &EmitterState()
   &SpawnBurst_Instantaneous(Use Spawn Probability=false, UseLoopCountLimit=false)
  ]
  particleSpawnScript [
   /SetVariables(Particles.VisibilityTag=$Emitter.SpawnGroup) {isSetParameters=true}
   &InitializeParticle(
    Color=<rgba 0 352.0194 500 1>
    Lifetime=$System.IntroLength
    Ribbon Width=10
    Ribbon Width Mode=::NewEnumerator1
    Sprite Rotation Mode=::NewEnumerator0
    Sprite Size Mode=::NewEnumerator3
    Uniform Sprite Size=&Multiply_Float(A=18, B=&FN_FrontEndFloatModifier(Gameplay=1, Front End=$User.FrontEndScale))
    UsePositionOffset=false
    Write Material Random=true
   )
   &CameraOffset(Camera Offset Amount=5)
  ]
  particleUpdateScript [
   &ParticleState()
   &ScaleSpriteSize(
    Scale Sprite Size Mode=::NewEnumerator0
    Uniform Scale Factor=&Add_Float(
     A=0
     B=&FloatFromCurve(
      FloatCurve=DataInterface() {propertyValues {Curve {
         defaultValue=0
         keys [
          <at 0 0.3 RCIM_Cubic>
          {arriveTangent=-0.221, interpMode=RCIM_Cubic, leaveTangent=-0.221, tangentMode=RCTM_User, time=0.3899, value=1}
          {arriveTangent=0.1935, interpMode=RCIM_Cubic, leaveTangent=0.1935, time=0.6996, value=1.3101}
          {arriveTangent=6.566, interpMode=RCIM_Cubic, leaveTangent=6.566, tangentMode=RCTM_User, time=0.9067, value=1.1}
          {arriveTangent=35.933, interpMode=RCIM_Cubic, leaveTangent=35.933, tangentMode=RCTM_User, time=1, value=4}
         ]
        }}}
      Scale Curve=1.2
     )
    )
   ) {module=ScaleSpriteSize002}
   &ScaleColor(
    Scale Alpha=&FloatFromCurve(FloatCurve=DataInterface() {propertyValues {Curve {
        defaultValue=0
        keys [
         {arriveTangent=9.0725, interpMode=RCIM_Cubic, leaveTangent=9.0725, tangentMode=RCTM_User, time=0, value=0}
         {arriveTangent=-0.2441, leaveTangent=-0.2441, time=0.1408, value=0.9847}
         {arriveTangent=-1.146, leaveTangent=-1.146, time=0.95, value=1}
         {arriveTangent=-0.0405, interpMode=RCIM_Cubic, leaveTangent=-0.0405, tangentMode=RCTM_User, time=1, value=0}
        ]
       }}})
    ScaleA=true
    ScaleRGB=true
   )
   &OffsetPosition(Position Offset=&Lerp_Vector(
     A=$System.HeartIntroSpawnPosition
     B=$System.BurstPosition
     Alpha=&FloatFromCurve(FloatCurve=DataInterface() {propertyValues {Curve {keys [
          {arriveTangent=4.0014, interpMode=RCIM_Cubic, leaveTangent=4.0014, tangentMode=RCTM_User, time=0, value=0}
          {arriveTangent=-0.1074, tangentMode=RCTM_User, time=1, value=1}
         ]}}})
    ))
   &DynamicMaterialParameters(
    Index 0 Param 1=&FloatFromCurve(
     FloatCurve=DataInterface() {propertyValues {Curve {keys [<at 0.5196 1.8 RCIM_Cubic>,<at 0.6196 2 RCIM_Cubic>,<at 0.77 1.8 RCIM_Cubic>]}}}
     Scale Curve=0.6
    )
    Index 0 Param 2=&FloatFromCurve(
     FloatCurve=DataInterface() {propertyValues {Curve {keys [<at 0.5196 1 RCIM_Cubic>,<at 0.6196 2 RCIM_Cubic>,<at 0.77 1 RCIM_Cubic>]}}}
     Scale Curve=1.15
    )
    Index 0 Param 3=&FloatFromCurve(
     FloatCurve=DataInterface() {propertyValues {Curve {keys [<at 0.5196 1.8 RCIM_Cubic>,<at 0.6196 2 RCIM_Cubic>,<at 0.77 1.8 RCIM_Cubic>]}}}
     Scale Curve=1.4
    )
    Module.Index0Float0Write=true
    Module.Index0Float1Write=true
    Module.Index0Float2Write=true
    Module.Index0Float3Write=true
   )
   /SetVariables(
    Particles.Glow Color=&Multiply_LinearColorByFloat(
     Linear Color=&MultiplyLinearColor(
      Linear Color=<rgba 0.4427 0.3008 0.0548 0.35>
      Scale Factor=&MakeVector4_FromLinearColor(LinearColor=&FNTimeOfDayColor(Night Color=<rgba 0.2843 0.389 0.6823 1>, Morning Color=<rgba 0.8 0.7542 0.7542 1>, Day Color=<rgba 1 1 1 1>, Evening Color=<rgba 1 1 1 1>))
     )
     Float=&FN_FrontEndFloatModifier(Gameplay=0.6, Front End=0.45)
    )
    Particles.Glow Size=&Multiply_Vector2D(A=$Particles.SpriteSize, B=<1.5 1.5>)
   ) {isSetParameters=true, module=SetVariables_#9}
   &ScaleRibbonWidth(Ribbon Width Scale=&FloatFromCurve(FloatCurve=DataInterface() {propertyValues {Curve {keys [<at 0 1>,<at 0.2 1>,<at 0.7 1>,<at 1 0>]}}}))
  ]
  rendererClasses [&NiagaraSpriteRendererProperties]
  renderers [
   {rendererClass=&NiagaraSpriteRendererProperties}
   {rendererClass=&NiagaraSpriteRendererProperties, rendererIndex=1}
  ]
 }
 MainTrail {
  emitterSpawnScript [/SetVariables(Emitter.RandomOffset:NiagaraFloat=&RandomRangeFloat(Minimum=0, Maximum=30)) {isSetParameters=true, module=SetVariables_#10}]
  emitterUpdateScript [
   &EmitterState()
   &SpawnRate(
    SpawnGroup=1
    SpawnRate=&FloatFromCurve(
     FloatCurve=DataInterface() {propertyValues {Curve {keys [<at 0 1 RCIM_Constant>,<at 0.2046 0 RCIM_Constant>]}}}
     CurveIndex=$Emitter.NormalizedLoopAge
     Scale Curve=30
    )
    Use Spawn Probability=false
   )
  ]
  particleSpawnScript [
   &InitializeParticle(
    Lifetime=0.4
    Ribbon Width=&Multiply_Float(A=20, B=&FN_FrontEndFloatModifier(Gameplay=1, Front End=$User.FrontEndScale))
    Ribbon Width Mode=::NewEnumerator1
    Sprite Rotation Mode=::NewEnumerator0
    UsePositionOffset=false
    Write Material Random=true
   )
   /SetVariables(Particles.Position=&Add_Vector(
     A=<0 0 4>
     B=&Lerp_Vector(
      A=$System.HeartIntroSpawnPosition
      B=$System.BurstPosition
      Alpha=&FloatFromCurve(
       FloatCurve=DataInterface() {propertyValues {Curve {keys [
           {arriveTangent=4.0014, interpMode=RCIM_Cubic, leaveTangent=4.0014, tangentMode=RCTM_User, time=0, value=0}
           {arriveTangent=-0.1073, tangentMode=RCTM_User, time=1, value=1}
          ]}}}
       CurveIndex=&Divide_Float(A=$Emitter.Age, B=1.3)
      )
     )
    )) {isSetParameters=true}
   &AddVelocity(
    Velocity=&MakeVector(
     X=&Waveform(
      Global Input Position=$Emitter.RandomOffset
      Global Phase=&Multiply_FloatByInt(Float=0.01, Integer=&ReturnParticleIDAttributes())
      "Waveform [1]"=::NewEnumerator0
      "[1] Frequency"=12
      "[1] Amplitude Min/Max"=<-1 1>
      "[1] Scale"=5
     )
     Y=&MakeFloatFromVector(
      Vector=&MakeVector(
       X=&Waveform(
        Global Input Position=$Emitter.RandomOffset
        Global Phase=&Multiply_FloatByInt(Float=0.01, Integer=&ReturnParticleIDAttributes())
        "Waveform [1]"=::NewEnumerator1
        "[1] Frequency"=12
        "[1] Amplitude Min/Max"=<-1 1>
        "[1] Scale"=5
       )
       Y=0
       Z=0
      )
      Channel=::NewEnumerator0
     )
     Z=&RandomRangeFloat(Minimum=0, Maximum=-5)
    )
    Velocity Speed Scale=&FloatFromCurve(
     FloatCurve=DataInterface() {propertyValues {Curve {keys [<at 0.5 1>,<at 1 0>]}}}
     CurveIndex=&Divide_Float(A=$Emitter.Age, B=$System.IntroLength)
     Scale Curve=&RandomRangeFloat(Minimum=3, Maximum=5)
    )
   )
   /SetVariables(Particles.RibbonUVDistance:NiagaraFloat=&DistanceBetweenPositions(
     A=&Lerp_Vector(
      A=$System.HeartIntroSpawnPosition
      B=$System.BurstPosition
      Alpha=&FloatFromCurve(
       FloatCurve=DataInterface() {propertyValues {Curve {keys [
           {arriveTangent=0.7913, interpMode=RCIM_Cubic, leaveTangent=0.7913, tangentMode=RCTM_User, time=0, value=0}
           {arriveTangent=-0.1073, tangentMode=RCTM_User, time=1, value=1}
          ]}}}
       CurveIndex=&Divide_Float(A=$Emitter.Age, B=$System.IntroLength)
      )
     )
     B=<0 0 0>
    )) {isSetParameters=true}
   /SetVariables(Particles.SpawnAge:NiagaraFloat=&Divide_Float(A=$Emitter.Age, B=$System.IntroLength)) {isSetParameters=true}
  ]
  particleUpdateScript [
   &ParticleState()
   &ScaleColor(
    Scale Alpha=&Multiply_Float(
     A=&FloatFromCurve(
      FloatCurve=DataInterface() {propertyValues {Curve {
         defaultValue=0
         keys [
          <at 0 0.3>
          <at 0.1155 1>
          {arriveTangent=-0.2441, time=0.5, value=1}
          {arriveTangent=-0.0405, interpMode=RCIM_Cubic, leaveTangent=-0.0405, tangentMode=RCTM_User, time=1, value=0}
         ]
        }}}
      Scale Curve=0.7
     )
     B=&FloatFromCurve(
      FloatCurve=DataInterface() {propertyValues {Curve {keys [
          <at 0.1 0 RCIM_Cubic>
          <at 0.25 1>
          {arriveTangent=-1.7056, interpMode=RCIM_Cubic, leaveTangent=-1.7056, time=0.5804, value=1}
          <at 0.8363 0 RCIM_Cubic>
         ]}}}
      CurveIndex=$System.Age
     )
    )
    Scale RGB=&MakeVectorFromLinearColorRGB(Color=&Multiply_LinearColorByFloat(
      Linear Color=&ColorFromCurve(ColorCurve=DataInterface() {propertyValues {
         AlphaCurve {keys [<at 0 1>]}
         BlueCurve {keys [<at 0 0.0831>, {arriveTangent=-0.4043, time=0.5015, value=0.0664}, <at 1 0.0795>]}
         GreenCurve {keys [<at 0 0.4787>, {arriveTangent=0.0031, time=0.5015, value=0.3611}, <at 1 0.3869>]}
         RedCurve {keys [<at 0 0.7292>, {arriveTangent=-0.2602, time=0.5015, value=0.5729}, <at 1 0.6042>]}
        }})
      Float=2
     ))
    ScaleA=true
    ScaleRGB=true
   )
   &DynamicMaterialParameters(
    Index 0 Param 1=-0.5
    Index 0 Param 3=1
    Index 0 Param 4=1
    Module.Index0Float0Write=true
    Module.Index0Float1Write=true
    Module.Index0Float2Write=true
    Module.Index0Float3Write=true
   ) {module=DynamicMaterialParameters001}
   &ScaleRibbonWidth(Ribbon Width Scale=&Multiply_Float(
     A=&FloatFromCurve(FloatCurve=DataInterface() {propertyValues {Curve {keys [<at 0 1>,<at 0.3 1>]}}})
     B=&FloatFromCurve(FloatCurve=DataInterface() {propertyValues {Curve {keys [<at 0.05 0>,<at 0.15 1>,<at 0.5 1>,<at 0.9 0>]}}}, CurveIndex=$Particles.SpawnAge)
    ))
   &Drag(Drag=2, Use Linear Drag=true, Use Rotational Drag=true)
   &SolveForcesAndVelocity(Clamp Velocity=false, Limit Acceleration=false)
  ]
  rendererClasses [&NiagaraRibbonRendererProperties]
  renderers [{rendererClass=&NiagaraRibbonRendererProperties}]
 }
 Pop {
  emitterUpdateScript [
   &EmitterState()
   &SpawnBurst_Instantaneous(
    Spawn Count=2
    Spawn Time=&Subtract_Float(A=$System.IntroLength, B=0.1)
    Use Spawn Probability=false
    UseLoopCountLimit=false
   )
  ]
  particleSpawnScript [
   &FN_SetSpawnGroup(Number of Spawn Groups=2)
   &InitializeParticle(
    Color=<rgba 5 4.0463 0.7327 0.8>
    Lifetime=&MergeHelper_FloatSwitch(0=0.25, 1=0.47, 2=?, 3=?, Index=&SpawnGroup(), Num=::NewEnumerator0)
    Position Offset=$System.BurstPosition
    Sprite Rotation Mode=::NewEnumerator0
    Sprite Size Mode=::NewEnumerator3
    Uniform Sprite Size=&MergeHelper_FloatSwitch(
     0=&Multiply_Float(A=20, B=&FN_FrontEndFloatModifier(Gameplay=1, Front End=$User.FrontEndScale))
     1=&RandomRangeFloat(Minimum=50, Maximum=60)
     2=?
     3=?
     Index=&SpawnGroup()
     Num=::NewEnumerator0
    )
    UsePositionOffset=true
    Write Material Random=true
   )
   &DynamicMaterialParameters(
    Index 0 Param 4=1
    Module.Index0Float0Write=true
    Module.Index0Float1Write=true
    Module.Index0Float2Write=true
    Module.Index0Float3Write=true
   )
   &CameraOffset(Camera Offset Amount=&MergeHelper_FloatSwitch(0=0, 1=-5, 2=?, 3=?, Index=&SpawnGroup(), Num=::NewEnumerator0))
   /SetVariables(Particles.VisibilityTag=&SpawnGroup()) {isSetParameters=true}
   /SetVariables(Particles.DynDust:Vector4f=&MakeVector4(X=0.08, Y=&RandomRangeFloat(Minimum=0, Maximum=14), Z=0, W=0)) {isSetParameters=true}
  ]
  particleUpdateScript [
   &ParticleState()
   &GravityForce(Gravity=&MergeHelper_VectorSwitch(0=<0 0 0>, 1=<0 0 -50>, 2=?, 3=?, Index=&SpawnGroup(), Num=::NewEnumerator0))
   &ScaleSpriteSize(
    Scale Sprite Size Mode=::NewEnumerator0
    Uniform Scale Factor=&MergeHelper_FloatSwitch(
     0=&FloatFromCurve(
      FloatCurve=DataInterface() {propertyValues {Curve {keys [
          {arriveTangent=3.3458, interpMode=RCIM_Cubic, leaveTangent=3.3458, tangentMode=RCTM_User, time=0, value=1}
          {arriveTangent=1.251, tangentMode=RCTM_User, time=1, value=3}
         ]}}}
      Scale Curve=1.05
     )
     1=&FloatFromCurve(
      FloatCurve=DataInterface() {propertyValues {Curve {keys [
          {arriveTangent=1.2481, interpMode=RCIM_Cubic, leaveTangent=1.2481, tangentMode=RCTM_User, time=0, value=0.5}
          {interpMode=RCIM_Cubic, tangentMode=RCTM_User, time=1, value=1}
         ]}}}
      Scale Curve=&Multiply_Float(A=1.2, B=&FN_FrontEndFloatModifier(Gameplay=1, Front End=0.65))
     )
     2=?
     3=?
     Index=&SpawnGroup()
     Num=::NewEnumerator0
    )
   ) {module=ScaleSpriteSize002}
   &ScaleColor(
    Scale Alpha=&FloatFromCurve(FloatCurve=DataInterface() {propertyValues {Curve {
        defaultValue=0
        keys [
         {arriveTangent=0.3137, interpMode=RCIM_Cubic, leaveTangent=0.3137, tangentMode=RCTM_User, time=0, value=1}
         {arriveTangent=-2.0958, interpMode=RCIM_Cubic, leaveTangent=-2.0958, tangentMode=RCTM_User, time=1, value=0}
        ]
       }}})
    Scale RGB=&VectorFromCurve(VectorCurve=DataInterface() {propertyValues {
       XCurve {keys [{interpMode=RCIM_Cubic, tangentMode=RCTM_User, time=0, value=1}, {arriveTangent=-0.2294, tangentMode=RCTM_User, time=1, value=0}]}
       YCurve {keys [{interpMode=RCIM_Cubic, tangentMode=RCTM_User, time=0, value=1}, {arriveTangent=-0.2294, tangentMode=RCTM_User, time=1, value=0}]}
       ZCurve {keys [{interpMode=RCIM_Cubic, tangentMode=RCTM_User, time=0, value=1}, {arriveTangent=-0.2294, tangentMode=RCTM_User, time=1, value=0}]}
      }})
    ScaleA=true
    ScaleRGB=true
   )
   &SolveForcesAndVelocity(Clamp Velocity=false, Limit Acceleration=false)
   /SetVariables(Particles.Glow Color=&Multiply_LinearColorByFloat(
     Linear Color=&MultiplyLinearColor(
      Linear Color=$Particles.Color
      Scale Factor=&MakeVector4_FromLinearColor(LinearColor=<rgba 1 0.628 0.1453 1>)
     )
     Float=0.33
    )) {isSetParameters=true}
   /SetVariables(Particles.DustColor:LinearColor=&MakeLinearColorFromVectorAndFloat(
     "Vector (RGB)"=&MakeVectorFromLinearColorRGB(Color=&MultiplyLinearColor(
       Linear Color=<rgba 1 0.7162 0.1894 1>
       Scale Factor=&MakeVector4_FromLinearColor(LinearColor=&FNTimeOfDayColor(Night Color=<rgba 0.2843 0.389 0.6823 1>, Morning Color=<rgba 0.8 0.7542 0.7542 1>, Day Color=<rgba 1 1 1 1>, Evening Color=<rgba 1 1 1 1>))
      ))
     "Float (Alpha)"=&FloatFromCurve(
      FloatCurve=DataInterface() {propertyValues {Curve {keys [
          {arriveTangent=3.8913, interpMode=RCIM_Cubic, leaveTangent=3.8913, tangentMode=RCTM_User, time=0, value=0.6}
          {arriveTangent=-0.2539, interpMode=RCIM_Cubic, leaveTangent=-0.2539, tangentMode=RCTM_User, time=0.1489, value=1}
          {arriveTangent=-0.4437, interpMode=RCIM_Cubic, leaveTangent=-0.4437, tangentMode=RCTM_User, time=1, value=0}
         ]}}}
      Scale Curve=0.5
     )
    )) {isSetParameters=true}
  ]
  rendererClasses [&NiagaraSpriteRendererProperties]
  renderers [
   {rendererClass=&NiagaraSpriteRendererProperties}
   {rendererClass=&NiagaraSpriteRendererProperties, rendererIndex=1}
  ]
 }
 SmallHeartsTrail {
  emitterSpawnScript [/SetVariables(Emitter.RandomOffset:NiagaraFloat=&RandomRangeFloat(Minimum=0, Maximum=30)) {isSetParameters=true, module=SetVariables_#10}]
  emitterUpdateScript [
   &EmitterState()
   &FN_SpawnAnimTrail(Number of Anim Trails=5)
   &SpawnRate(
    SpawnGroup=1
    SpawnRate=&FloatFromCurve(
     FloatCurve=DataInterface() {propertyValues {Curve {keys [<at 0.5 0 RCIM_Constant>,<at 0.5035 1 RCIM_Constant>,<at 1 0 RCIM_Constant>]}}}
     CurveIndex=$Emitter.NormalizedLoopAge
     Scale Curve=30
    )
    Use Spawn Probability=false
   )
  ]
  particleSpawnScript [
   &FN_InitializeAnimTrailBoneIndecies()
   &FN_InitializeAnimTrail(
    AnimTrailWidth=4.3
    Position Offset=&Lerp_Vector(
     A=&Add_Vector(A=$System.BurstPosition, B=<0 0 3>)
     B=&Add_Vector(
      A=$System.BurstPosition
      B=&Multiply_Vector(
       A=<1 1 1>
       B=&MergeHelper_VectorSwitch(
        0=$System.RandomVectorPos1
        1=$System.RandomVectorPos2
        2=$System.RandomVectorPos3
        3=$System.RandomVectorPos4
        4=$System.RandomVectorPos5
        Index=&SpawnGroup()
        Num=::NewEnumerator3
       )
      )
     )
     Alpha=&MergeHelper_FloatSwitch(
      0=&FloatFromCurve(
       FloatCurve=DataInterface() {propertyValues {Curve {keys [
           {arriveTangent=5.0192, interpMode=RCIM_Cubic, leaveTangent=5.0192, tangentMode=RCTM_User, time=0.54, value=0}
           {arriveTangent=3.6846, tangentMode=RCTM_User, time=0.85, value=0.98}
          ]}}}
       CurveIndex=&Divide_Float(A=$Emitter.Age, B=1.2)
      )
      1=&FloatFromCurve(
       FloatCurve=DataInterface() {propertyValues {Curve {keys [
           {arriveTangent=5.0192, interpMode=RCIM_Cubic, leaveTangent=5.0192, tangentMode=RCTM_User, time=0.54, value=0}
           {arriveTangent=3.6846, tangentMode=RCTM_User, time=0.87, value=0.98}
          ]}}}
       CurveIndex=&Divide_Float(A=$Emitter.Age, B=1.2)
      )
      2=&FloatFromCurve(
       FloatCurve=DataInterface() {propertyValues {Curve {keys [
           {arriveTangent=5.0192, interpMode=RCIM_Cubic, leaveTangent=5.0192, tangentMode=RCTM_User, time=0.54, value=0}
           {arriveTangent=3.6846, tangentMode=RCTM_User, time=0.91, value=0.98}
          ]}}}
       CurveIndex=&Divide_Float(A=$Emitter.Age, B=1.2)
      )
      3=&FloatFromCurve(
       FloatCurve=DataInterface() {propertyValues {Curve {keys [
           {arriveTangent=5.0192, interpMode=RCIM_Cubic, leaveTangent=5.0192, tangentMode=RCTM_User, time=0.54, value=0}
           {arriveTangent=3.6846, tangentMode=RCTM_User, time=0.9, value=0.98}
          ]}}}
       CurveIndex=&Divide_Float(A=$Emitter.Age, B=1.2)
      )
      4=&FloatFromCurve(
       FloatCurve=DataInterface() {propertyValues {Curve {keys [
           {arriveTangent=5.0192, interpMode=RCIM_Cubic, leaveTangent=5.0192, tangentMode=RCTM_User, time=0.54, value=0}
           {arriveTangent=3.6846, tangentMode=RCTM_User, time=0.98, value=0.98}
          ]}}}
       CurveIndex=&Divide_Float(A=$Emitter.Age, B=1.2)
      )
      Index=&SpawnGroup()
      Num=::NewEnumerator3
     )
    )
   )
  ]
  particleUpdateScript [
   &ParticleState()
   &ScaleColor(
    Scale Alpha=&Multiply_Float(
     A=&FloatFromCurve(
      FloatCurve=DataInterface() {propertyValues {Curve {
         defaultValue=0
         keys [
          <at 0 0.3 RCIM_Cubic>
          {arriveTangent=1.4, interpMode=RCIM_Cubic, leaveTangent=1.4, time=0.1155, value=1}
          {arriveTangent=-1.1306, interpMode=RCIM_Cubic, leaveTangent=-1.1306, time=0.5, value=1}
          {arriveTangent=-0.0405, interpMode=RCIM_Cubic, leaveTangent=-0.0405, tangentMode=RCTM_User, time=1, value=0}
         ]
        }}}
      Scale Curve=0.7
     )
     B=&FloatFromCurve(
      FloatCurve=DataInterface() {propertyValues {Curve {keys [<at 0 1 RCIM_Cubic>, {arriveTangent=-1, interpMode=RCIM_Cubic, leaveTangent=-1, time=0.9422, value=1}, <at 1 0 RCIM_Cubic>]}}}
      CurveIndex=$System.Age
     )
    )
    Scale RGB=&MakeVectorFromLinearColorRGB(Color=&Multiply_LinearColorByFloat(
      Linear Color=&ColorFromCurve(ColorCurve=DataInterface() {propertyValues {
         AlphaCurve {keys [<at 0 1>]}
         BlueCurve {keys [<at 0 0.1787>, {arriveTangent=-0.4043, time=0.5015, value=0.0664}, <at 1 0.0795>]}
         GreenCurve {keys [<at 0 1.1219>, {arriveTangent=0.0031, time=0.5015, value=0.3611}, <at 1 0.3869>]}
         RedCurve {keys [<at 0 2>, {arriveTangent=-0.2602, time=0.5015, value=0.5729}, <at 1 0.6042>]}
        }})
      Float=2
     ))
    ScaleA=true
    ScaleRGB=true
   )
   &ScaleRibbonWidth(Ribbon Width Scale=&Multiply_Float(
     A=&FloatFromCurve(FloatCurve=DataInterface() {propertyValues {Curve {keys [<at 0 1>,<at 1 0.4>]}}})
     B=1
    ))
   &SolveForcesAndVelocity(Clamp Velocity=false, Limit Acceleration=false)
   &FN_UpdateAnimTrail()
  ]
  rendererClasses [&NiagaraRibbonRendererProperties]
  renderers [{rendererClass=&NiagaraRibbonRendererProperties}]
 }
}
summary {
 systemName=NS_ReactFX_FlourCut
 emitters [
  {emitterName=Intro, rendererClasses [&NiagaraSpriteRendererProperties]}
  {emitterName=Pop, rendererClasses [&NiagaraSpriteRendererProperties]}
  {emitterName=MainTrail, rendererClasses [&NiagaraRibbonRendererProperties]}
  {emitterName=Hearts, rendererClasses [&NiagaraSpriteRendererProperties]}
  {emitterName=SmallHeartsTrail, rendererClasses [&NiagaraRibbonRendererProperties]}
 ]
 userVariables [
  User.bIsFrontEnd {type=NiagaraBool}
  User.bIsFrontEndPreview {type=NiagaraBool}
  User.FrontEndScale {type=NiagaraFloat, defaultValue=0.6, description=""}
  User.RibbonIDOffset {type=NiagaraInt32}
 ]
}
