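# Niagara reaction effect for cosmetic companion petting: a pink glowing heart bubble travels to a burst point and pops.
# NE_Intro plays the sound and drives sprites, mesh and light along a curved lerp between system positions, with Intro_Sparkles following the same path.
# Pop fires delayed bursts at the burst point, using spawn groups to vary motion and visibility tags.
# Time-of-day functions tune colors and light, and user flags mark front-end contexts.
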NiagaraSystem /CosmeticCompanions/FX/React/NS_ReactFX_FondCuddles
dependencies {
 usedDataInterfaces [
  &NiagaraDataInterfaceAudioPlayer
  &NiagaraDataInterfaceCurve
  &NiagaraDataInterfaceColorCurve
  &NiagaraDataInterfaceVectorCurve
  &NiagaraDataInterfaceVector2DCurve
 ]
 usedDynamicInputs [
  &VectorFromCurve
  /Niagara/DynamicInputs/TypeConversions/MakeCustomPositionFromBool
  &FN_FrontEndVector3Modifier
 ]
 usedModules [
  &self:SystemSpawnScript.SystemScriptSource.SystemScriptGraph.NiagaraNodeAssignment_0.SetVariables_#1
  &self:SystemSpawnScript.SystemScriptSource.SystemScriptGraph.NiagaraNodeAssignment_1.SetVariables_#2
  &self:SystemSpawnScript.SystemScriptSource.SystemScriptGraph.NiagaraNodeAssignment_5.SetVariables_#3
  &self:SystemSpawnScript.SystemScriptSource.SystemScriptGraph.NiagaraNodeAssignment_4.SetVariables_#4
  &self:SystemSpawnScript.SystemScriptSource.SystemScriptGraph.NiagaraNodeAssignment_6.SetVariables_#5
  &SystemState
  &self:SystemSpawnScript.SystemScriptSource.SystemScriptGraph.NiagaraNodeAssignment_2.SetVariables_#6
  &self:SystemSpawnScript.SystemScriptSource.SystemScriptGraph.NiagaraNodeAssignment_3.SetVariables_#7
  &self:SystemSpawnScript.SystemScriptSource.SystemScriptGraph.NiagaraNodeAssignment_8.SetVariables_#8
  &self:SystemSpawnScript.SystemScriptSource.SystemScriptGraph.NiagaraNodeAssignment_9.SetVariables_#9
 ]
}
emitters {
 Intro_Sparkles {
  emitterUpdateScript [
   &EmitterState(
    Inactive Response=::NewEnumerator1
    Life Cycle Mode=::NewEnumerator1
    Loop Behavior=::NewEnumerator1
    Loop Delay=0.47
    Loop Duration=$System.IntroLength
    UseLoopDelay=false
   )
   &SpawnRate(SpawnRate=25, Use Spawn Probability=false)
  ]
  particleSpawnScript [
   &InitializeParticle(
    Color Maximum=<rgba 1 0.222 0.7039 1>
    Color Minimum=<rgba 1 0.4045 0.8917 1>
    Color Mode=::NewEnumerator2
    Lifetime Max=0.15
    Lifetime Min=0.1
    Lifetime Mode=::NewEnumerator1
    Sprite Size Max=<4 5>
    Sprite Size Min=<3 4>
    Sprite Size Mode=::NewEnumerator2
    UsePositionOffset=true
    Write Material Random=true
   )
   &CameraOffset(Camera Offset Amount=10)
   /SetVariables(Particles.VisibilityTag=&SpawnGroup()) {isSetParameters=true}
  ]
  particleUpdateScript [
   &ParticleState()
   &ShapeLocation(
    MaskSpawnGroup=false
    Non Uniform Scale=<1 1 0.8>
    Offset=&Lerp_Vector(
     A=$System.HeartIntroSpawnPosition
     B=$System.BurstPosition
     Alpha=&FloatFromCurve(
      FloatCurve=DataInterface() {propertyValues {Curve {keys [
          {arriveTangent=0.7913, interpMode=RCIM_Cubic, leaveTangent=0.7913, tangentMode=RCTM_User, time=0, value=0}
          {arriveTangent=0.2071, interpMode=RCIM_Cubic, leaveTangent=0.2071, tangentMode=RCTM_User, time=0.7, value=0.9784}
          <at 1 1 RCIM_Cubic>
         ]}}}
      CurveIndex=$Emitter.NormalizedLoopAge
     )
    )
    Offset Mode=::NewEnumerator0
    Override Seed=false
    Sphere Radius=10
    UseOwnerScale=true
   )
   &ScaleSpriteSize(
    Non-Uniform Curve Sprite Scale=DataInterface() {propertyValues {
      XCurve {keys [
        <at 0 0 RCIM_Cubic>
        {arriveTangentWeight=0.0872, interpMode=RCIM_Cubic, leaveTangentWeight=0.3197, tangentMode=RCTM_User, time=0.1569, value=1.625}
        {arriveTangentWeight=0.0926, interpMode=RCIM_Cubic, leaveTangentWeight=0.2166, tangentMode=RCTM_User, time=0.3622, value=1}
        {arriveTangentWeight=0.2824, interpMode=RCIM_Cubic, leaveTangentWeight=0.1558, tangentMode=RCTM_User, time=0.9005, value=1}
        <at 1.0034 0 RCIM_Cubic>
       ]}
      YCurve {keys [
        <at 0 0 RCIM_Cubic>
        {arriveTangentWeight=0.0872, interpMode=RCIM_Cubic, leaveTangentWeight=0.3197, tangentMode=RCTM_User, time=0.1534, value=1.625}
        {arriveTangentWeight=0.0676, interpMode=RCIM_Cubic, leaveTangentWeight=0.163, tangentMode=RCTM_User, time=0.3553, value=1}
        {arriveTangentWeight=0.2837, interpMode=RCIM_Cubic, leaveTangentWeight=0.153, tangentMode=RCTM_User, time=0.9005, value=1}
        <at 0.9966 2 RCIM_Cubic>
       ]}
     }}
    Scale Sprite Size Mode=::NewEnumerator3
   )
   &ScaleColor(
    Scale Alpha=&FloatFromCurve(
     FloatCurve=DataInterface() {propertyValues {Curve {keys [
         <at 0 0 RCIM_Cubic>
         {arriveTangent=0.4409, interpMode=RCIM_Cubic, leaveTangent=0.4409, time=0.1023, value=1}
         {arriveTangent=-1.114, interpMode=RCIM_Cubic, leaveTangent=-1.114, time=0.6137, value=0.2706}
         <at 1 0 RCIM_Cubic>
        ]}}}
     Scale Curve=&RandomRangeFloat(Minimum=0.5, Maximum=1)
    )
    ScaleA=true
    ScaleRGB=true
   )
  ]
  rendererClasses [&NiagaraSpriteRendererProperties]
  renderers [{rendererClass=&NiagaraSpriteRendererProperties}]
 }
 NE_Intro {
  emitterSpawnScript [
   &InitializeAudioPlayer(Audio Player=DataInterface() {propertyValues {
      Attenuation=null
      Concurrency=null
      MaxPlaysPerTick=10
      SoundToPlay=/CosmeticCompanions/Assets/Audio_Shared/Interaction/Petting/SC_Mimosa_Interaction_React_Default
      bAllowLoopingOneShotSounds=false
      bLimitPlaysPerTick=true
      bOnlyActiveDuringGameplay=false
      bStopWhenComponentIsDestroyed=true
      ConfigurationUserParameter {parameter {name=None, typeDefHandle {registeredTypeIndex=1}, varData []}}
      ParameterNames []
     }}) {module=InitializeAudioPlayer001}
   /Niagara/Modules/Audio/PlayAudio() {module=PlayAudio001}
   /SetVariables(Emitter.Heart Emissive:LinearColor=&FNTimeOfDayColor(
     Night Color=<rgba 0.2188 0.1125 0.0481 1>
     Morning Color=<rgba 0.1042 0.0157 0.0048 1>
     Day Color=<rgba 0.0208 0.0208 0.0208 1>
     Evening Color=<rgba 0.1562 0.0803 0.0344 1>
    )) {isSetParameters=true}
  ]
  emitterUpdateScript [
   &EmitterState()
   &SpawnBurst_Instantaneous(Spawn Group=3, Use Spawn Probability=false, UseLoopCountLimit=false) {module=SpawnBurst_Instantaneous003}
   &SpawnBurst_Instantaneous(Spawn Group=2, Use Spawn Probability=false, UseLoopCountLimit=false) {module=SpawnBurst_Instantaneous002}
   &SpawnBurst_Instantaneous(Spawn Group=1, Use Spawn Probability=false, UseLoopCountLimit=false) {module=SpawnBurst_Instantaneous001}
   &SpawnBurst_Instantaneous(Use Spawn Probability=false, UseLoopCountLimit=false)
  ]
  particleSpawnScript [
   &InitializeParticle(
    Color=&FNTimeOfDayColor(
     Night Color=<rgba 0.6174 0 0.3824 0.38>
     Morning Color=<rgba 1 0 0.6194 0.3391>
     Day Color=<rgba 1 0 0.6194 0.4>
     Evening Color=<rgba 1 0 0.6194 0.7016>
    )
    Lifetime=$System.IntroLength
    Mesh Scale Mode=::NewEnumerator3
    Mesh Uniform Scale=0.3
    Sprite Rotation Mode=::NewEnumerator0
    Sprite Size=<17.8183 15.7951>
    Sprite Size Mode=::NewEnumerator4
    UsePositionOffset=false
    Write Material Random=true
   )
   /SetVariables(Particles.BubbleColor:LinearColor=&FNTimeOfDayColor(
     Night Color=<rgba 12 6.3241 11.9069 1>
     Morning Color=<rgba 20 10.5402 19.8448 1>
     Day Color=<rgba 20 10.5402 19.8448 1>
     Evening Color=<rgba 20 10.5402 19.8448 1>
    )) {isSetParameters=true}
   &CameraOffset(Camera Offset Amount=&MergeHelper_FloatSwitch(0=0, 1=3, 2=3, 3=1, Index=&SpawnGroup(), Num=::NewEnumerator2))
   /SetVariables(Particles.VisibilityTag=&SpawnGroup()) {isSetParameters=true}
   /SetVariables(Particles.Hard Glow Color:LinearColor=<rgba 2 0.75 1.8781 1>, Particles.Hard Glow Size:Vector2f=<22.4538 23.2151>) {isSetParameters=true}
  ]
  particleUpdateScript [
   &ParticleState()
   &Light_Attributes(
    Color=1
    Color Value To Scale:LinearColor=$Particles.BubbleColor
    Diffuse Scale=2
    Enabled=1
    Falloff Exponent=2
    Light Radius=15
    Position=1
    Position Offset Coordinate Space:ENiagaraCoordinateSpace=::Local
    Position to Offset:NiagaraPosition=&Lerp_Vector(
     A=$System.HeartIntroSpawnPosition
     B=$System.BurstPosition
     Alpha=&FloatFromCurve(FloatCurve=DataInterface() {propertyValues {Curve {keys [
          {arriveTangent=0.7913, interpMode=RCIM_Cubic, leaveTangent=0.7913, tangentMode=RCTM_User, time=0, value=0}
          {arriveTangent=0.2071, interpMode=RCIM_Cubic, leaveTangent=0.2071, tangentMode=RCTM_User, time=0.7, value=0.9784}
          <at 1 1 RCIM_Cubic>
         ]}}})
    )
    Radius=2
    Scale Alpha:NiagaraFloat=&FNTimeOfDayScalar(Night Scale=0.1, Morning Scale=0.1, Day Scale=0.05, Evening Scale=0.1)
    ScaleA=true
    ScaleRGB=true
    Specular Scale=2
    UsePositionOffset=true
    Volumetric Scattering=2
   )
   &ScaleMeshSize(Scale Factor=&Multiply_VectorByFloat(
     Vector=<1 1 1>
     Float=&FloatFromCurve(FloatCurve=DataInterface() {propertyValues {Curve {keys [<at 0 0>,<at 0.4 1>,<at 0.95 1.0736>,<at 1.0045 1.5>]}}})
    )) {module=ScaleMeshSize001}
   &ScaleSpriteSize(
    Scale Sprite Size Mode=::NewEnumerator0
    Uniform Scale Factor=&Add_Float(
     A=1
     B=&FloatFromCurve(FloatCurve=DataInterface() {propertyValues {Curve {
         defaultValue=0
         keys [
          {arriveTangent=0.8193, interpMode=RCIM_Cubic, leaveTangent=0.8193, tangentMode=RCTM_User, time=-0.0405, value=-0.3866}
          {arriveTangent=0.6646, interpMode=RCIM_Cubic, leaveTangent=0.6646, tangentMode=RCTM_User, time=0.7, value=0.9545}
          {arriveTangent=-0.0367, interpMode=RCIM_Cubic, leaveTangent=-0.0367, tangentMode=RCTM_User, time=0.95, value=1.009}
          {arriveTangent=-0.1202, leaveTangent=-0.1202, time=1, value=2}
         ]
        }}})
    )
   ) {module=ScaleSpriteSize002}
   &ScaleColor(
    Scale Alpha=&FloatFromCurve(
     FloatCurve=DataInterface() {propertyValues {Curve {
        defaultValue=0
        keys [
         {arriveTangent=1.0331, interpMode=RCIM_Cubic, leaveTangent=1.0331, tangentMode=RCTM_User, time=0.35, value=0}
         {arriveTangent=-0.2441, time=0.7, value=0.9847}
         <at 0.893 0.9967>
         {arriveTangent=-1.146, leaveTangent=-1.146, time=1.0165, value=1.2588}
         {arriveTangent=-0.0405, interpMode=RCIM_Cubic, leaveTangent=-0.0405, tangentMode=RCTM_User, time=1.0356, value=0.125}
        ]
       }}}
     Scale Curve=4
    )
    ScaleA=true
    ScaleRGB=true
   )
   &OffsetPosition(Position Offset=&Lerp_Vector(
     A=$System.HeartIntroSpawnPosition
     B=$System.BurstPosition
     Alpha=&FloatFromCurve(FloatCurve=DataInterface() {propertyValues {Curve {keys [
          {arriveTangent=0.7913, interpMode=RCIM_Cubic, leaveTangent=0.7913, tangentMode=RCTM_User, time=0, value=0}
          {arriveTangent=0.2071, interpMode=RCIM_Cubic, leaveTangent=0.2071, tangentMode=RCTM_User, time=0.7, value=0.9784}
          <at 1 1 RCIM_Cubic>
         ]}}})
    ))
   &DynamicMaterialParameters(
    Index 0 Param 2=0.5245
    Module.Index0Float0Write=true
    Module.Index0Float1Write=true
    Module.Index0Float2Write=true
    Module.Index0Float3Write=true
    Module.Index1Float0Write=true
    Module.Index1Float1Write=true
    Module.Index1Float2Write=true
    Module.Index1Float3Write=true
    Write Parameter Index 1=::NewEnumerator1
   )
   /SetVariables(
    Particles.Glow Color=&Multiply_LinearColorByFloat(Linear Color=$System.Layer 2 Color, Float=0.7)
    Particles.Glow Size=&Multiply_Vector2D(A=$Particles.SpriteSize, B=<0.6 0.6>)
   ) {isSetParameters=true}
   /SetVariables(
    Particles.Scale Hard Glow Color:LinearColor=&MultiplyLinearColor(
     Linear Color=$Particles.Initial.Hard Glow Color
     Scale Factor=&Vector4FromCurve(Vector4Curve=DataInterface() {propertyValues {
        AlphaCurve {keys [<at 0 0>,<at 0.0233 1>,<at 0.9684 1>,<at 1 0>]}
        BlueCurve {keys [<at 0 1>]}
        GreenCurve {keys [<at 0 1>]}
        RedCurve {keys [<at 0 1>]}
       }})
    )
    Particles.Scale Hard Glow Size:Vector2f=&Multiply_Vector2D(
     A=$Particles.Initial.Hard Glow Size
     B=&Vector2DFromFloat(Value=&FloatFromCurve(FloatCurve=DataInterface() {propertyValues {Curve {
          defaultValue=0
          keys [
           {arriveTangent=0.8193, time=0, value=0}
           {arriveTangent=0.5017, interpMode=RCIM_Cubic, leaveTangent=0.5017, tangentMode=RCTM_User, time=0.3799, value=0.914}
           {arriveTangent=-0.0367, interpMode=RCIM_Cubic, leaveTangent=-0.0367, tangentMode=RCTM_User, time=0.95, value=1.009}
           {arriveTangent=-0.1202, leaveTangent=-0.1202, time=1, value=1.5}
          ]
         }}}))
    )
   ) {isSetParameters=true}
  ]
  rendererClasses [
   &NiagaraSpriteRendererProperties
   &NiagaraMeshRendererProperties
   &NiagaraLightRendererProperties
  ]
  renderers [
   {rendererClass=&NiagaraSpriteRendererProperties}
   {rendererClass=&NiagaraMeshRendererProperties, rendererIndex=1}
   {rendererClass=&NiagaraLightRendererProperties, rendererIndex=2}
   {rendererClass=&NiagaraSpriteRendererProperties, rendererIndex=3}
  ]
 }
 Pop {
  emitterUpdateScript [
   &EmitterState()
   &SpawnBurst_Instantaneous(Spawn Group=5, Spawn Time=0.71, Use Spawn Probability=false, UseLoopCountLimit=false) {module=SpawnBurst_Instantaneous005}
   &SpawnBurst_Instantaneous(Spawn Group=4, Spawn Time=0.71, Use Spawn Probability=false, UseLoopCountLimit=false) {module=SpawnBurst_Instantaneous004}
   &SpawnBurst_Instantaneous(Spawn Group=3, Spawn Time=0.71, Use Spawn Probability=false, UseLoopCountLimit=false) {module=SpawnBurst_Instantaneous003}
   &SpawnBurst_Instantaneous(Spawn Count=8, Spawn Group=1, Spawn Time=0.7, Use Spawn Probability=false, UseLoopCountLimit=false)
   &SpawnBurst_Instantaneous(Spawn Count=8, Spawn Time=0.7, Use Spawn Probability=false, UseLoopCountLimit=false) {module=SpawnBurst_Instantaneous001}
   &SpawnBurst_Instantaneous(Spawn Count=12, Spawn Group=2, Spawn Time=0.73, Use Spawn Probability=false, UseLoopCountLimit=false) {module=SpawnBurst_Instantaneous002}
  ]
  particleSpawnScript [
   /SetVariables(Particles.VisibilityTag=&SpawnGroup()) {isSetParameters=true}
   &InitializeParticle(
    Color=&FNTimeOfDayColor(
     Night Color=&RandomRangeLinearColor(Minimum=<rgba 1 0.2604 0.8995 0.5>, Maximum=<rgba 1 0.099 0.6571 0.5>)
     Morning Color=&RandomRangeLinearColor(Minimum=<rgba 1 0.2604 0.8995 0.7>, Maximum=<rgba 1 0.099 0.6571 0.7>)
     Day Color=&RandomRangeLinearColor(Minimum=<rgba 1 0.2604 0.8995 1>, Maximum=<rgba 1 0.099 0.6571 1>)
     Evening Color=&RandomRangeLinearColor(Minimum=<rgba 1 0.2604 0.8995 0.7>, Maximum=<rgba 1 0.099 0.6571 0.7>)
    )
    Lifetime=&MergeHelper_FloatSwitch(
     0=&RandomRangeFloat(Minimum=0.25, Maximum=0.8)
     1=&RandomRangeFloat(Minimum=0.18, Maximum=0.3)
     2=&RandomRangeFloat(Minimum=0.25, Maximum=0.55)
     3=0.15
     4=0.22
     5=0.15
     Index=&SpawnGroup()
     Num=::NewEnumerator4
    )
    Mesh Scale Mode=::NewEnumerator1
    Mesh Uniform Scale Max=0.3
    Mesh Uniform Scale Min=0.15
    Position Offset=$System.BurstPosition
    Sprite Size Max=<15 18>
    Sprite Size Min=<4 7>
    Sprite Size Mode=::NewEnumerator2
    UsePositionOffset=true
    Write Material Random=true
   )
   &AddVelocity(
    Module.Use Falloff Distance=false
    Override Seed=false
    Velocity Mode=::NewEnumerator1
    Velocity Origin:Vector3f=$System.BurstPosition
    Velocity Speed=&MergeHelper_FloatSwitch(
     0=&RandomRangeFloat(Minimum=70, Maximum=110)
     1=&RandomRangeFloat(Minimum=80, Maximum=200)
     2=&RandomRangeFloat(Minimum=40, Maximum=100)
     3=0
     4=0
     5=0
     Index=&SpawnGroup()
     Num=::NewEnumerator4
    )
   )
   /SetVariables(Particles.Glow Size=&Vector2DFromFloat(Value=&RandomRangeFloat(Minimum=2, Maximum=8))) {isSetParameters=true}
   /SetVariables(
    Particles.AirGlow Dynam:Vector4f=&MakeVector4(X=2, Y=0, Z=0, W=0)
    Particles.Airglow Color:LinearColor=<rgba 0.099 0.7793 1 0.5>
    Particles.Airglow Size:Vector2f=&Vector2DFromFloat(Value=65)
   ) {isSetParameters=true}
   /SetVariables(
    Particles.Rays Color:LinearColor=<rgba 1 0.2457 0.7302 1>
    Particles.Rays Dynam:Vector4f=&MakeVector4(
     X=&RandomRangeFloat(Minimum=0, Maximum=10)
     Y=&RandomRangeFloat(Minimum=2.5, Maximum=4.5)
     Z=&RandomRangeFloat(Minimum=0.8, Maximum=1.6)
     W=0
    )
    Particles.Rays Size:Vector3f=&VectorFromFloat(Value=0.4)
   ) {isSetParameters=true}
   /SetVariables(
    Particles.Corona Color:LinearColor=<rgba 1 0.7672 0.4943 1>
    Particles.Corona Size:Vector2f=&Vector2DFromFloat(Value=55)
   ) {isSetParameters=true}
   &CameraOffset(Camera Offset Amount=&MergeHelper_FloatSwitch(0=0, 1=0, 2=0, 3=0, 4=15, Index=&SpawnGroup(), Num=::NewEnumerator3))
  ]
  particleUpdateScript [
   &ParticleState()
   &ScaleMeshSize(Scale Factor=&VectorFromFloat(Value=&FloatFromCurve(FloatCurve=DataInterface() {propertyValues {Curve {keys [<at 0 1>,<at 0.7964 1>,<at 1 0>]}}})))
   &ScaleSpriteSize(
    Scale Sprite Size Mode=::NewEnumerator0
    Uniform Scale Factor=&FloatFromCurve(
     FloatCurve=DataInterface() {propertyValues {Curve {keys [
         <at 0 0 RCIM_Cubic>
         {arriveTangentWeight=0.0433, interpMode=RCIM_Cubic, leaveTangentWeight=0.3986, tangentMode=RCTM_User, time=0.0852, value=1}
         {arriveTangentWeight=0.3878, interpMode=RCIM_Cubic, leaveTangentWeight=0.064, tangentMode=RCTM_User, time=0.8704, value=1}
         <at 1 0 RCIM_Cubic>
        ]}}}
     Scale Curve=&MakeFloatFromVector2d(Vector 2D=<1 2>)
    )
   ) {module=ScaleSpriteSize002}
   &ScaleColor(
    Linear Color Curve=DataInterface() {propertyValues {
      AlphaCurve {keys [<at 0 0>,<at 0.1364 0.5>,<at 0.4198 0.5>,<at 1 0>]}
      BlueCurve {keys [<at 0 0.901>]}
      GreenCurve {keys [<at 0 0.901>]}
      RedCurve {keys [<at 0 0.901>]}
     }}
    Scale Mode=::NewEnumerator2
   )
   &Drag(
    Drag=&MergeHelper_FloatSwitch(
     0=&FloatFromCurve(
      FloatCurve=DataInterface() {propertyValues {Curve {keys [
          {interpMode=RCIM_Cubic, leaveTangentWeight=7.3294, tangentMode=RCTM_User, time=0.2181, value=0}
          {
           arriveTangent=119.2046
           arriveTangentWeight=7.3294
           interpMode=RCIM_Cubic
           leaveTangent=119.2046
           tangentMode=RCTM_User
           time=0.4999
           value=10
          }
         ]}}}
      Scale Curve=&RandomRangeFloat(Minimum=0.5, Maximum=1)
     )
     1=0
     2=&FloatFromCurve(
      FloatCurve=DataInterface() {propertyValues {Curve {keys [
          {interpMode=RCIM_Cubic, leaveTangentWeight=7.3294, tangentMode=RCTM_User, time=0.2181, value=0}
          {arriveTangent=119.2046, arriveTangentWeight=7.3294, interpMode=RCIM_Cubic, leaveTangent=119.2046, tangentMode=RCTM_User, time=0.35, value=20}
         ]}}}
      Scale Curve=&RandomRangeFloat(Minimum=0.5, Maximum=1)
     )
     3=0
     4=0
     5=0
     Index=&SpawnGroup()
     Num=::NewEnumerator4
    )
    Use Linear Drag=true
    Use Rotational Drag=true
   )
   &GravityForce(Gravity=&MergeHelper_VectorSwitch(
     0=<0 0 0>
     1=<0 0 0>
     2=&RandomRangeVector(Minimum=<0 0 150>, Maximum=<1 1 400>)
     3=<0 0 0>
     4=<0 0 0>
     Index=&SpawnGroup()
     Num=::NewEnumerator3
    ))
   &SolveForcesAndVelocity(Clamp Velocity=false, Limit Acceleration=false)
   /SetVariables(
    Particles.Glow Color=&Multiply_LinearColorByFloat(
     Linear Color=&MultiplyLinearColor(
      Linear Color=$Particles.Color
      Scale Factor=&MakeVector4_FromLinearColor(LinearColor=<rgba 1 0.222 0.7039 1>)
     )
     Float=0.33
    )
    Particles.Scale Glow Size:Vector2f=&Multiply_Vector2D(
     A=$Particles.Initial.Glow Size
     B=&Vector2DFromCurve(
      Vector2Curve=DataInterface() {propertyValues {
        XCurve {keys [
          <at 0 0 RCIM_Cubic>
          {arriveTangentWeight=0.0872, interpMode=RCIM_Cubic, leaveTangentWeight=0.3197, tangentMode=RCTM_User, time=0.1015, value=2}
          {arriveTangent=-0.1459, interpMode=RCIM_Cubic, leaveTangent=-0.1459, tangentMode=RCTM_User, time=0.2551, value=1}
          {arriveTangent=-1.065, interpMode=RCIM_Cubic, leaveTangent=-1.065, time=0.4549, value=1.375}
          {arriveTangent=-0.7773, interpMode=RCIM_Cubic, leaveTangent=-0.7773, time=0.6073, value=0.625}
          {arriveTangent=1.281, interpMode=RCIM_Cubic, leaveTangent=1.281, time=0.7765, value=1.125}
          {arriveTangentWeight=0.2824, interpMode=RCIM_Cubic, leaveTangentWeight=0.1558, tangentMode=RCTM_User, time=0.9, value=1}
          <at 1.0034 0 RCIM_Cubic>
         ]}
        YCurve {keys [
          <at 0 0 RCIM_Cubic>
          {arriveTangentWeight=0.0872, interpMode=RCIM_Cubic, leaveTangentWeight=0.3197, tangentMode=RCTM_User, time=0.1015, value=2}
          {arriveTangent=-0.1459, interpMode=RCIM_Cubic, leaveTangent=-0.1459, tangentMode=RCTM_User, time=0.2551, value=1}
          {arriveTangent=-1.0549, interpMode=RCIM_Cubic, leaveTangent=-1.0549, time=0.4515, value=1.375}
          {arriveTangent=-0.7613, interpMode=RCIM_Cubic, leaveTangent=-0.7613, time=0.6106, value=0.625}
          {arriveTangent=1.296, interpMode=RCIM_Cubic, leaveTangent=1.296, time=0.7799, value=1.125}
          {arriveTangentWeight=0.2837, interpMode=RCIM_Cubic, leaveTangentWeight=0.153, tangentMode=RCTM_User, time=0.9, value=1}
          <at 0.9966 3 RCIM_Cubic>
         ]}
       }}
      ScaleCurve=<1 1.3>
     )
    )
   ) {isSetParameters=true}
   /SetVariables(Particles.Scale Airglow Color:LinearColor=&Multiply_LinearColorByFloat(
     Linear Color=$Particles.Initial.Airglow Color
     Float=&FloatFromCurve(FloatCurve=DataInterface() {propertyValues {Curve {keys [<at 0 0 RCIM_Cubic>,<at 0.25 1 RCIM_Cubic>,<at 1 0 RCIM_Cubic>]}}})
    )) {isSetParameters=true}
   /SetVariables(
    Particles.Scale Rays Color:LinearColor=&Multiply_LinearColorByFloat(
     Linear Color=$Particles.Initial.Rays Color
     Float=&FloatFromCurve(FloatCurve=DataInterface() {propertyValues {Curve {keys [
          <at 0 0 RCIM_Cubic>
          {arriveTangent=0.4409, interpMode=RCIM_Cubic, leaveTangent=0.4409, time=0.1023, value=1}
          {arriveTangent=-1.114, interpMode=RCIM_Cubic, leaveTangent=-1.114, time=0.6137, value=0.2706}
          <at 1 0 RCIM_Cubic>
         ]}}})
    )
    Particles.Scale Rays Size:Vector3f=&Multiply_VectorByFloat(
     Vector=$Particles.Initial.Rays Size
     Float=&FloatFromCurve(FloatCurve=DataInterface() {propertyValues {Curve {keys [
          <at 0 0 RCIM_Cubic>
          {arriveTangentWeight=0.0571, interpMode=RCIM_Cubic, leaveTangentWeight=0.2418, tangentMode=RCTM_User, time=0.0938, value=1}
          {arriveTangent=-0.4414, interpMode=RCIM_Cubic, leaveTangent=-0.4414, time=0.4913, value=0.75}
          <at 1 0.6 RCIM_Cubic>
         ]}}})
    )
   ) {isSetParameters=true}
   /SetVariables(
    Particles.Scale Corona Color:LinearColor=&Multiply_LinearColorByFloat(
     Linear Color=$Particles.Initial.Corona Color
     Float=&FloatFromCurve(FloatCurve=DataInterface() {propertyValues {Curve {keys [<at 0 0.45>,<at 0.2 1>,<at 0.5 0.25>,<at 1 0>]}}})
    )
    Particles.Scale Corona Size:Vector2f=&Multiply_Vector2D(
     A=$Particles.Initial.Corona Size
     B=&Vector2DFromFloat(Value=&FloatFromCurve(FloatCurve=DataInterface() {propertyValues {Curve {keys [<at 0 0 RCIM_Cubic>, {arriveTangent=1, interpMode=RCIM_Cubic, leaveTangent=1, time=0.108, value=0.6}, <at 1 1 RCIM_Cubic>]}}}))
    )
   ) {isSetParameters=true}
  ]
  rendererClasses [&NiagaraSpriteRendererProperties, &NiagaraMeshRendererProperties]
  renderers [
   {rendererClass=&NiagaraSpriteRendererProperties}
   {rendererClass=&NiagaraSpriteRendererProperties, rendererIndex=1}
   {rendererClass=&NiagaraMeshRendererProperties, rendererIndex=2}
   {rendererClass=&NiagaraSpriteRendererProperties, rendererIndex=3}
  ]
 }
}
summary {
 systemName=NS_ReactFX_FondCuddles
 emitters [
  {
   emitterName=NE_Intro
   rendererClasses [
    &NiagaraSpriteRendererProperties
    &NiagaraMeshRendererProperties
    &NiagaraLightRendererProperties
   ]
  }
  {emitterName=Pop, rendererClasses [&NiagaraSpriteRendererProperties, &NiagaraMeshRendererProperties]}
  {emitterName=Intro_Sparkles, rendererClasses [&NiagaraSpriteRendererProperties]}
 ]
 userVariables [
  User.bIsFrontEnd {type=NiagaraBool}
  User.bIsFrontEndPreview {type=NiagaraBool}
 ]
}
