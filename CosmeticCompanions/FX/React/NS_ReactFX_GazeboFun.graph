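# Niagara reaction effect for cosmetic companions: a sprite pops up with a reaction sound, then bursts with hearts.
# Pookie scales the main sprite by age and swaps its SubUV frame, tinted by time of day or front-end preview.
# Pop001 and HeartBurst add brief flashes, while Hearts fires two outlined bursts from a camera-oriented ring, slowed by drag.
# Emitters share System.BurstPosition; Pop001 and Hearts time their bursts off System.IntroLength.
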NiagaraSystem /CosmeticCompanions/FX/React/NS_ReactFX_GazeboFun
dependencies {
 usedDataInterfaces [
  &NiagaraDataInterfaceCamera
  &NiagaraDataInterfaceCurve
  &NiagaraDataInterfaceColorCurve
  &NiagaraDataInterfaceVector2DCurve
  &NiagaraDataInterfaceVectorCurve
  &NiagaraDataInterfaceAudioPlayer
  &NiagaraDataInterfaceSpriteRendererInfo
 ]
 usedDynamicInputs [&FN_FrontEndVector3Modifier]
 usedModules [
  &self:SystemSpawnScript.SystemScriptSource.SystemScriptGraph.NiagaraNodeAssignment_0.SetVariables_#1
  &self:SystemSpawnScript.SystemScriptSource.SystemScriptGraph.NiagaraNodeAssignment_8.SetVariables_#2
  &self:SystemSpawnScript.SystemScriptSource.SystemScriptGraph.NiagaraNodeAssignment_9.SetVariables_#3
  &self:SystemSpawnScript.SystemScriptSource.SystemScriptGraph.NiagaraNodeAssignment_5.SetVariables_#4
  &self:SystemSpawnScript.SystemScriptSource.SystemScriptGraph.NiagaraNodeAssignment_6.SetVariables_#5
  &self:SystemSpawnScript.SystemScriptSource.SystemScriptGraph.NiagaraNodeAssignment_7.SetVariables_#6
  &self:SystemSpawnScript.SystemScriptSource.SystemScriptGraph.NiagaraNodeAssignment_3.SetVariables_#7
  &self:SystemSpawnScript.SystemScriptSource.SystemScriptGraph.NiagaraNodeAssignment_2.SetVariables_#8
  &SystemState
 ]
}
emitters {
 HeartBurst {
  emitterUpdateScript [
   &EmitterState()
   &SpawnBurst_Instantaneous(Spawn Time=0.7, Use Spawn Probability=false, UseLoopCountLimit=false)
  ]
  particleSpawnScript [
   &InitializeParticle(
    Lifetime=0.1
    Position Offset=$System.BurstPosition
    Sprite Size=&Multiply_Vector2D_ByFloat(Float=50, Vector2D=<1 1>)
    Sprite Size Mode=::NewEnumerator4
    UsePositionOffset=true
    Write Material Random=true
   )
   &CameraOffset(Camera Offset Amount=20)
  ]
  particleUpdateScript [
   &ParticleState()
   &ScaleSpriteSize(Uniform Curve Sprite Scale=DataInterface() {propertyValues {Curve {keys [{arriveTangent=0.7298, interpMode=RCIM_Cubic, leaveTangent=0.7298, tangentMode=RCTM_User, time=0, value=0.2122}, <at 1 0.8 RCIM_Cubic>]}}})
   &ScaleColor(
    Scale Alpha=&FloatFromCurve(FloatCurve=DataInterface() {propertyValues {Curve {
        defaultValue=1
        keys [
         {interpMode=RCIM_Cubic, tangentMode=RCTM_User, time=0, value=1}
         {arriveTangent=-2.52, interpMode=RCIM_Cubic, leaveTangent=-2.52, tangentMode=RCTM_User, time=1, value=0}
        ]
       }}})
    Scale RGB=&MakeVectorFromLinearColorRGB(Color=&ColorFromCurve(ColorCurve=DataInterface() {propertyValues {
        AlphaCurve {keys []}
        BlueCurve {keys [<at 0 0.1927 RCIM_Cubic>,<at 0.5046 0.0369 RCIM_Cubic>]}
        GreenCurve {keys [<at 0 0.2311 RCIM_Cubic>,<at 0.5046 0.0222 RCIM_Cubic>]}
        RedCurve {keys [<at 0 2 RCIM_Cubic>,<at 0.5046 0.9216 RCIM_Cubic>]}
       }}))
    ScaleA=true
    ScaleRGB=true
   )
  ]
  rendererClasses [&NiagaraSpriteRendererProperties]
  renderers [{rendererClass=&NiagaraSpriteRendererProperties}]
 }
 Hearts {
  emitterUpdateScript [
   &EmitterState()
   &SpawnBurst_Instantaneous(
    Spawn Count=10
    Spawn Time=&Subtract_Float(A=$System.IntroLength, B=0.04)
    Use Spawn Probability=false
    UseLoopCountLimit=false
   )
   &SpawnBurst_Instantaneous(Spawn Count=10, Spawn Group=1, Spawn Time=$System.IntroLength, Use Spawn Probability=false, UseLoopCountLimit=false) {module=SpawnBurst_Instantaneous001}
  ]
  particleSpawnScript [
   /SetVariables(Particles.VisibilityTag=&SpawnGroup()) {isSetParameters=true}
   &InitializeParticle(
    Color=&FNTimeOfDayColor(
     Night Color=<rgba 0.4832 0.7459 1 1>
     Morning Color=<rgba 1 0.8695 0.7944 1>
     Day Color=<rgba 1 1 1 1>
     Evening Color=<rgba 1 0.8695 0.7944 1>
    )
    Lifetime=&MaskFloat(
     Mask Source=0
     Mask Parameter=?
     Mask=0
     Masked Float=&RandomRangeFloat(Minimum=0.3, Maximum=0.6)
     Passthrough Float=0.4
    )
    Position Offset=$System.BurstPosition
    Sprite Size=/Niagara/DynamicInputs/Helpers/MaskVector2d(
     Mask Source:NiagaraInt32=0
     Mask Parameter:NiagaraInt32=?
     Mask:NiagaraInt32=0
     Masked Vector2D:Vector2f=&Multiply_Vector2D_ByFloat(Float=1.5, Vector2D=&RandomRangeVector2D(Minimum=<18 14>, Maximum=<38 30>))
     Passthrough Vector2D:Vector2f=<15 10>
    )
    Sprite Size Mode=::NewEnumerator4
    UsePositionOffset=true
    Write Material Random=true
   )
   &ShapeLocation(
    Additional Quaternion Rotation=true
    Additional Yaw / Pitch / Roll=<0 0.25 0>
    Handle Radius=1
    Large Radius=5
    MaskSpawnGroup=false
    Rotation Mode=::NewEnumerator3
    Rotation Quaternion=&OrientMeshToVector(
     Look At Direction=&ReturnCameraVector(Camera Vectors=::NewEnumerator1, Coordinate Space=::Simulation)
     Reference Axis Vector=<1 0 0>
    )
    Shape Primitive=::NewEnumerator3
    Torus Distribution Mode:ENiagaraTorusDistributionMode=::NewEnumerator1
    U Position=&MergeHelper_FloatSwitch(
     0=0
     1=0.1444
     2=0.2711
     3=0.3444
     4=0.4356
     5=0.4644
     6=0.5378
     7=0.6556
     8=0.7733
     9=0.8467
     Index=&ReturnExecIndex()
     Num=::NewEnumerator8
    )
    UseOwnerScale=true
   )
   &AddVelocity(
    Constrain To Radius=100
    Module.Use Falloff Distance=false
    Origin Offset=<0 0 1>
    Override Seed=false
    Velocity Mode=::NewEnumerator1
    Velocity Origin:Vector3f=$System.BurstPosition
    Velocity Speed=115
   )
   &CameraOffset(Camera Offset Amount=15)
  ]
  particleUpdateScript [
   &ParticleState()
   /SetVariables(Particles.Flicker:NiagaraFloat=&Waveform(
     "Waveform [1]"=::NewEnumerator0
     "[1] Frequency"=&RandomRangeFloat(Minimum=3, Maximum=5)
     "[1] Amplitude Min/Max"=<0 1>
    )) {isSetParameters=true}
   &Drag(
    Drag=&Multiply_Float(
     A=&RandomRangeFloat(Minimum=3, Maximum=7)
     B=&FloatFromCurve(FloatCurve=DataInterface() {propertyValues {Curve {defaultValue=1, keys [{tangentMode=RCTM_User, time=0.2, value=0}, {tangentMode=RCTM_User, time=0.35, value=2}]}}})
    )
    Use Linear Drag=true
    Use Rotational Drag=true
   )
   &SolveForcesAndVelocity(Clamp Velocity=false, Limit Acceleration=false)
   &ScaleColor(
    Scale Alpha=&MaskFloat(
     Mask Source=0
     Mask Parameter=?
     Mask=0
     Masked Float=&FloatFromCurve(FloatCurve=DataInterface() {propertyValues {Curve {
         defaultValue=1
         keys [
          {interpMode=RCIM_Cubic, tangentMode=RCTM_User, time=0.2721, value=1}
          {arriveTangent=-2.8008, interpMode=RCIM_Cubic, leaveTangent=-2.8008, tangentMode=RCTM_User, time=1, value=0}
         ]
        }}})
     Passthrough Float=&FloatFromCurve(
      FloatCurve=DataInterface() {propertyValues {Curve {
         defaultValue=1
         keys [
          {arriveTangent=-2.463, interpMode=RCIM_Cubic, leaveTangent=-2.463, tangentMode=RCTM_User, time=0.3842, value=1}
          {arriveTangent=-0.1815, interpMode=RCIM_Cubic, leaveTangent=-0.1815, tangentMode=RCTM_User, time=0.7919, value=0}
         ]
        }}}
      Scale Curve=10
     )
    )
    Scale RGB=&MaskVector(
     Mask Source=0
     Mask Parameter=?
     Mask=0
     Masked Vector=&MakeVectorFromLinearColorRGB(Color=&ColorFromCurve(ColorCurve=DataInterface() {propertyValues {
         AlphaCurve {keys []}
         BlueCurve {keys [<at 0 0.0801 RCIM_Cubic>,<at 1 0.0369 RCIM_Cubic>]}
         GreenCurve {keys [<at 0 0.0481 RCIM_Cubic>,<at 1 0.0222 RCIM_Cubic>]}
         RedCurve {keys [<at 0 2 RCIM_Cubic>,<at 1 0.9216 RCIM_Cubic>]}
        }}))
     Passthrough Vector=&MakeVectorFromLinearColorRGB(Color=<rgba 0 0 0 1>)
    )
    ScaleA=true
    ScaleRGB=true
   )
   /SetVariables(Particles.Hearts SpriteSize:Vector2f=&Multiply_Vector2D(
     A=&Vector2DFromCurve(
      Vector2Curve=DataInterface() {propertyValues {XCurve {keys [<at 0.1 1>,<at 0.3777 0.55>,<at 0.9 1>]}, YCurve {keys [<at 0.1 1>,<at 0.3754 1.4662>,<at 0.9 1>]}}}
      ScaleCurve=<1 1>
     )
     B=&Vector2DFromFloat(Value=&FloatFromCurve(
       FloatCurve=DataInterface() {propertyValues {Curve {
          defaultValue=1
          keys [
           {arriveTangent=3.2566, interpMode=RCIM_Cubic, leaveTangent=3.2566, tangentMode=RCTM_User, time=0, value=0.5}
           {arriveTangent=-0.5, interpMode=RCIM_Cubic, leaveTangent=-0.5, time=0.5039, value=1.1092}
           {arriveTangent=0.0224, interpMode=RCIM_Cubic, leaveTangent=0.0224, tangentMode=RCTM_User, time=1, value=0}
          ]
         }}}
       Scale Curve=&Multiply_Float(A=1, B=&RandomRangeFloat(Minimum=8, Maximum=18))
      ))
    )) {isSetParameters=true}
   /SetVariables(
    Particles.OutlineColor:LinearColor=&MakeLinearColorFromVectorAndFloat(
     "Vector (RGB)"=<0 0 0>
     "Float (Alpha)"=&MaskFloat(
      Mask Source=0
      Mask Parameter=?
      Mask=0
      Masked Float=&FloatFromCurve(FloatCurve=DataInterface() {propertyValues {Curve {
          defaultValue=1
          keys [
           {interpMode=RCIM_Cubic, tangentMode=RCTM_User, time=0.2721, value=1}
           {arriveTangent=-2.8008, interpMode=RCIM_Cubic, leaveTangent=-2.8008, tangentMode=RCTM_User, time=1, value=0}
          ]
         }}})
      Passthrough Float=&FloatFromCurve(
       FloatCurve=DataInterface() {propertyValues {Curve {
          defaultValue=1
          keys [
           {arriveTangent=-2.463, interpMode=RCIM_Cubic, leaveTangent=-2.463, tangentMode=RCTM_User, time=0.3842, value=1}
           {arriveTangent=-0.1815, interpMode=RCIM_Cubic, leaveTangent=-0.1815, tangentMode=RCTM_User, time=0.7919, value=0}
          ]
         }}}
       Scale Curve=10
      )
     )
    )
    Particles.OutlineSize:Vector2f=&Multiply_Vector2D(A=$Particles.Hearts SpriteSize, B=&Multiply_Vector2D_ByFloat(Float=1.3, Vector2D=<1 1>))
   ) {isSetParameters=true}
  ]
  rendererClasses [&NiagaraSpriteRendererProperties]
  renderers [
   {rendererClass=&NiagaraSpriteRendererProperties}
   {rendererClass=&NiagaraSpriteRendererProperties, rendererIndex=1}
   {rendererClass=&NiagaraSpriteRendererProperties, rendererIndex=2}
  ]
 }
 Pookie {
  emitterSpawnScript [
   &InitializeAudioPlayer(Audio Player=DataInterface() {propertyValues {
      Attenuation=null
      Concurrency=null
      MaxPlaysPerTick=10
      SoundToPlay=/CosmeticCompanions/Assets/Audio_Shared/Interaction/Petting/SC_Mimosa_Interaction_React_Default
      bAllowLoopingOneShotSounds=false
      bLimitPlaysPerTick=true
      bOnlyActiveDuringGameplay=false
      bStopWhenComponentIsDestroyed=true
      ConfigurationUserParameter {parameter {name=None, typeDefHandle {registeredTypeIndex=1}, varData []}}
      ParameterNames []
     }})
   /Niagara/Modules/Audio/PlayAudio() {module=PlayAudio001}
  ]
  emitterUpdateScript [
   &EmitterState(Life Cycle Mode=::NewEnumerator1, Loop Behavior=::NewEnumerator1, Loop Duration=2, UseLoopDelay=false)
   &SpawnBurst_Instantaneous(Use Spawn Probability=false, UseLoopCountLimit=false)
  ]
  particleSpawnScript [
   &InitializeParticle(
    Color=&MultiplyLinearColor(
     Linear Color=&MakeCustomLinearColorFromBool(
      Bool=$User.bIsFrontEndPreview
      True Color=<rgba 0.5 0.5 0.5 1>
      False Color=&FNTimeOfDayColor(
       Night Color=<rgba 0.6117 0.7914 1 1>
       Morning Color=<rgba 0.8976 0.8104 0.7205 1>
       Day Color=<rgba 1 1 1 1>
       Evening Color=<rgba 0.9022 0.8146 0.7243 1>
      )
     )
     Scale Factor=&MakeVector4_FromLinearColor(LinearColor=<rgba 0.5747 1 0.8772 1>)
    )
    Lifetime=1.5
    Position Offset=$System.BurstPosition
    Position Offset Coordinate Space=::Local
    Sprite Size Mode=::NewEnumerator3
    Uniform Sprite Size=22
    UsePositionOffset=true
    Write Material Random=true
   )
   /SetVariables(Particles.RightRootSize:Vector2f=<35 35>) {isSetParameters=true}
   &CameraOffset(Camera Offset Amount=10)
  ]
  particleUpdateScript [
   &ParticleState()
   &ScaleSpriteSize(
    Scale Sprite Size Mode=::NewEnumerator0
    Uniform Scale Factor=&FloatFromCurve(
     FloatCurve=DataInterface() {propertyValues {Curve {
        defaultValue=0
        keys [
         {arriveTangent=3.7864, interpMode=RCIM_Cubic, leaveTangent=3.7864, tangentMode=RCTM_User, time=0, value=0}
         {arriveTangent=0.6849, interpMode=RCIM_Cubic, leaveTangent=0.6849, time=0.6345, value=1.0049}
         {arriveTangent=-0.1964, interpMode=RCIM_Cubic, leaveTangent=-0.1964, tangentMode=RCTM_User, time=0.73, value=0.5}
         {arriveTangent=1.865, interpMode=RCIM_Cubic, leaveTangent=1.865, tangentMode=RCTM_User, time=0.7647, value=1.5817}
         {
          arriveTangent=-2.5886
          arriveTangentWeight=0.4135
          interpMode=RCIM_Cubic
          leaveTangent=-2.5886
          leaveTangentWeight=0.1173
          tangentMode=RCTM_User
          time=1.0787
          value=1.5457
         }
         {arriveTangent=-20.5819, interpMode=RCIM_Cubic, leaveTangent=-20.5819, tangentMode=RCTM_User, time=1.2, value=0}
        ]
       }}}
     CurveIndex=$Particles.Age
    )
   )
   &ScaleColor(
    Scale Alpha=&FloatFromCurve(FloatCurve=DataInterface() {propertyValues {Curve {keys [<at 0 1>,<at 0.8938 1>,<at 1 0>]}}})
    ScaleA=true
    ScaleRGB=true
   )
   &SubUVAnimation(
    Direct Frame Index:NiagaraFloat=&MakeCustomFloatFromBool(Bool=&SetBoolByFloatComparison(A=$Particles.Age, B=0.75), True Float=1, False Float=0)
    "Sprite Renderer "=DataInterface() {propertyValues {SpriteRenderer=&self:Pookie_0.NiagaraSpriteRendererProperties_0}}
    SubUV Animation Mode:ENiagara_SubUVLookupModeV2=::NewEnumerator5
   )
  ]
  rendererClasses [&NiagaraSpriteRendererProperties]
  renderers [{rendererClass=&NiagaraSpriteRendererProperties}]
 }
 Pop001 {
  emitterUpdateScript [
   &EmitterState()
   &SpawnBurst_Instantaneous(
    Spawn Group=1
    Spawn Time=&Subtract_Float(A=$System.IntroLength, B=0.05)
    Use Spawn Probability=false
    UseLoopCountLimit=false
   ) {module=SpawnBurst_Instantaneous001}
   &SpawnBurst_Instantaneous(
    Spawn Time=&Subtract_Float(A=$System.IntroLength, B=0.05)
    Use Spawn Probability=false
    UseLoopCountLimit=false
   )
  ]
  particleSpawnScript [
   /SetVariables(Particles.VisibilityTag=&SpawnGroup()) {isSetParameters=true}
   &InitializeParticle(
    Lifetime=0.15
    Position Offset=$System.BurstPosition
    Sprite Rotation Mode=::NewEnumerator0
    Sprite Size Mode=::NewEnumerator3
    Uniform Sprite Size=&Multiply_Float(A=12, B=$System.Heart Size)
    UsePositionOffset=true
    Write Material Random=true
   )
   &DynamicMaterialParameters(
    Index 0 Param 4=1
    Module.Index0Float0Write=true
    Module.Index0Float1Write=true
    Module.Index0Float2Write=true
    Module.Index0Float3Write=true
   )
   /SetVariables(Particles.center Linear Color:LinearColor=<rgba 1 0.8431 0 1>, Particles.center spike SpriteRotation=0.0234) {isSetParameters=true}
   &CameraOffset(Camera Offset Amount=15)
  ]
  particleUpdateScript [
   &ParticleState()
   &ScaleSpriteSize(
    Scale Sprite Size Mode=::NewEnumerator0
    Uniform Scale Factor=&FloatFromCurve(
     FloatCurve=DataInterface() {propertyValues {Curve {keys [
         {arriveTangent=3.3458, interpMode=RCIM_Cubic, leaveTangent=3.3458, tangentMode=RCTM_User, time=0, value=1}
         {arriveTangent=1.251, tangentMode=RCTM_User, time=0.45, value=3}
        ]}}}
     Scale Curve=3
    )
   ) {module=ScaleSpriteSize002}
   &ScaleColor(
    Scale Alpha=&FloatFromCurve(FloatCurve=DataInterface() {propertyValues {Curve {
        defaultValue=0
        keys [
         {arriveTangent=0.3137, interpMode=RCIM_Cubic, leaveTangent=0.3137, tangentMode=RCTM_User, time=0.2, value=1}
         {arriveTangent=-2.0958, interpMode=RCIM_Cubic, leaveTangent=-2.0958, tangentMode=RCTM_User, time=1, value=0}
        ]
       }}})
    Scale RGB=&VectorFromCurve(VectorCurve=DataInterface() {propertyValues {
       XCurve {keys [{interpMode=RCIM_Cubic, tangentMode=RCTM_User, time=0, value=1}, {arriveTangent=-0.2294, tangentMode=RCTM_User, time=1, value=0}]}
       YCurve {keys [{interpMode=RCIM_Cubic, tangentMode=RCTM_User, time=0, value=1}, {arriveTangent=-0.2294, tangentMode=RCTM_User, time=1, value=0}]}
       ZCurve {keys [{interpMode=RCIM_Cubic, tangentMode=RCTM_User, time=0, value=1}, {arriveTangent=-0.2294, tangentMode=RCTM_User, time=1, value=0}]}
      }})
    ScaleA=true
    ScaleRGB=true
   )
   &SolveForcesAndVelocity(Clamp Velocity=false, Limit Acceleration=false)
   /SetVariables(
    Particles.Glow Color=&Multiply_LinearColorByFloat(
     Linear Color=&MultiplyLinearColor(
      Linear Color=$Particles.Color
      Scale Factor=&MakeVector4_FromLinearColor(LinearColor=<rgba 0.9216 0.0222 0.0369 1>)
     )
     Float=0.5
    )
    Particles.center position:NiagaraPosition=&AddVectorToPosition(Position=Linked() {linkedVariable {name=System.BurstPosition, type=Vector3f}}, Vector=<0 0 -0.7275>)
    Particles.center sprite size:Vector2f=&Vector2DFromFloat(Value=&FloatFromCurve(
      FloatCurve=DataInterface() {propertyValues {Curve {keys [
          {arriveTangent=3.3458, interpMode=RCIM_Cubic, leaveTangent=3.3458, tangentMode=RCTM_User, time=0, value=1}
          {arriveTangent=1.251, tangentMode=RCTM_User, time=0.5, value=3}
         ]}}}
      Scale Curve=6.5
     ))
   ) {isSetParameters=true}
   &DynamicMaterialParameters(
    Index 0 Param 3=&FloatFromCurve(FloatCurve=DataInterface() {propertyValues {Curve {keys [<at 0 1>,<at 1 0>]}}}, Scale Curve=2)
    Index 0 Param 4=&FloatFromCurve(FloatCurve=DataInterface() {propertyValues {Curve {keys [<at 0 1>,<at 0 1>,<at 0.9068 0.9891>,<at 1 0>]}}}, Scale Curve=5)
    Module.Index0Float0Write=true
    Module.Index0Float1Write=true
    Module.Index0Float2Write=false
    Module.Index0Float3Write=true
   ) {module=DynamicMaterialParameters001}
  ]
  rendererClasses [&NiagaraSpriteRendererProperties]
  renderers [
   {rendererClass=&NiagaraSpriteRendererProperties}
   {rendererClass=&NiagaraSpriteRendererProperties, rendererIndex=1}
   {rendererClass=&NiagaraSpriteRendererProperties, rendererIndex=2}
  ]
 }
}
summary {
 systemName=NS_ReactFX_GazeboFun
 emitters [
  {emitterName=Hearts, rendererClasses [&NiagaraSpriteRendererProperties]}
  {emitterName=Pop001, rendererClasses [&NiagaraSpriteRendererProperties]}
  {emitterName=Pookie, rendererClasses [&NiagaraSpriteRendererProperties]}
  {emitterName=HeartBurst, rendererClasses [&NiagaraSpriteRendererProperties]}
 ]
 userVariables [
  User.bIsFrontEnd {type=NiagaraBool}
  User.bIsFrontEndPreview {type=NiagaraBool}
 ]
}
