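# Niagara reaction effect for a cosmetic companion (HeadWhale) that gathers bubbles into a pop and plays a petting reaction sound.
# Bubble meshes are pulled toward System.BurstPosition, with ribbon trails sampled from one bubble emitter, while an intro sprite grows there.
# A pop flash with time-of-day tinting fires after IntroLength, alongside a timed burst of meshes and speed-stretched sprites.
# Intro size responds to front-end scale; two emitters are disabled.
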NiagaraSystem /CosmeticCompanions/FX/React/NS_ReactFX_HeadWhale
dependencies {
 usedDataInterfaces [&NiagaraDataInterfaceCurve, &NiagaraDataInterfaceParticleRead, &NiagaraDataInterfaceAudioPlayer]
 usedDynamicInputs [/Niagara/DynamicInputs/TypeConversions/MakeCustomPositionFromBool, &FN_FrontEndVector3Modifier]
 usedModules [
  &self:SystemSpawnScript.SystemScriptSource.SystemScriptGraph.NiagaraNodeAssignment_0.SetVariables_#1
  &self:SystemSpawnScript.SystemScriptSource.SystemScriptGraph.NiagaraNodeAssignment_1.SetVariables_#2
  &self:SystemSpawnScript.SystemScriptSource.SystemScriptGraph.NiagaraNodeAssignment_8.SetVariables_#3
  &self:SystemSpawnScript.SystemScriptSource.SystemScriptGraph.NiagaraNodeAssignment_7.SetVariables_#4
  &self:SystemSpawnScript.SystemScriptSource.SystemScriptGraph.NiagaraNodeAssignment_9.SetVariables_#5
  &self:SystemSpawnScript.SystemScriptSource.SystemScriptGraph.NiagaraNodeAssignment_4.SetVariables_#6
  &self:SystemSpawnScript.SystemScriptSource.SystemScriptGraph.NiagaraNodeAssignment_10.SetVariables_#7
  &self:SystemSpawnScript.SystemScriptSource.SystemScriptGraph.NiagaraNodeAssignment_2.SetVariables_#8
  &self:SystemSpawnScript.SystemScriptSource.SystemScriptGraph.NiagaraNodeAssignment_3.SetVariables_#9
  &self:SystemSpawnScript.SystemScriptSource.SystemScriptGraph.NiagaraNodeAssignment_5.SetVariables_#10
  &self:SystemSpawnScript.SystemScriptSource.SystemScriptGraph.NiagaraNodeAssignment_6.SetVariables_#11
  &SystemState
 ]
}
emitters {
 BubbleLoop {
  emitterUpdateScript [
   &EmitterState()
   &SpawnRate(
    SpawnRate=&FloatFromCurve(
     FloatCurve=DataInterface() {propertyValues {Curve {keys [<at 0 1 RCIM_Cubic>, {arriveTangent=-7.1525, interpMode=RCIM_Cubic, leaveTangent=-7.1525, tangentMode=RCTM_User, time=0.125, value=0}]}}}
     CurveIndex=$Emitter.NormalizedLoopAge
     Scale Curve=30
    )
    Use Spawn Probability=false
   )
  ]
  particleSpawnScript [
   &InitializeParticle(
    Color=<rgba 0.564 1 0.612 0.5>
    Mesh Scale Mode=::NewEnumerator1
    Mesh Uniform Scale Max=0.07
    Mesh Uniform Scale Min=0.025
    Position Offset=$System.BurstPosition
    Sprite Size Mode=::NewEnumerator1
    Uniform Sprite Size Max=12
    Uniform Sprite Size Min=7
    UsePositionOffset=true
    Write Material Random=true
   )
   &ShapeLocation(MaskSpawnGroup=false, Override Seed=false, Sphere Radius=32, Sphere Surface Distribution=1, UseOwnerScale=true)
  ]
  particleUpdateScript [
   &PointAttractionForce(
    Attraction Strength=50
    Attractor Position:Vector3f=$System.BurstPosition
    Kill Radius=5
    Kill Radius Overshoot Correction=3
    Kill Within Radius=true
    Module.Use Falloff=true
    UseAttractorOffset=false
   )
   &SolveForcesAndVelocity(Clamp Velocity=false, Limit Acceleration=false)
   &ParticleState()
   &ScaleMeshSizeBySpeed(Max Scale Factor=<2 0.35 0.35>, Speed Threshold=50)
   &DynamicMaterialParameters(
    Index 3 Param 2=&MakeFloatFromVector(Vector=$Particles.Scale, Channel=::NewEnumerator0)
    Index 3 Param 4=&RandomRangeFloat(Minimum=0, Maximum=100)
    Module.Param0WriteEnabled3=false
    Module.Param1WriteEnabled3=true
    Module.Param2WriteEnabled3=false
    Module.Param3WriteEnabled3=true
    Write Parameter Index 0=false
    Write Parameter Index 3=true
   )
   &ScaleColor(
    Scale Alpha=&Multiply_Float(
     A=&FloatFromCurve(FloatCurve=DataInterface() {propertyValues {Curve {keys [
          {interpMode=RCIM_Cubic, tangentMode=RCTM_User, time=0, value=0}
          {interpMode=RCIM_Cubic, tangentMode=RCTM_User, time=0.3, value=1}
          {tangentMode=RCTM_User, time=1, value=0}
         ]}}})
     B=1
    )
    Scale RGB=&Multiply_VectorByFloat(
     Vector=<1 1 1>
     Float=&FloatFromCurve(FloatCurve=DataInterface() {propertyValues {Curve {keys [
          {interpMode=RCIM_Cubic, tangentMode=RCTM_User, time=0, value=0}
          {interpMode=RCIM_Cubic, tangentMode=RCTM_User, time=0.3, value=1}
          {tangentMode=RCTM_User, time=1, value=0}
         ]}}})
    )
    ScaleA=true
    ScaleRGB=true
   )
   &ScaleMeshSize(Scale Factor=&Multiply_VectorByFloat(
     Vector=<1 1 1>
     Float=&FloatFromCurve(FloatCurve=DataInterface() {propertyValues {Curve {
         defaultValue=0.5
         keys [
          {arriveTangent=4.1785, interpMode=RCIM_Cubic, leaveTangent=4.1785, tangentMode=RCTM_User, time=0, value=0}
          {arriveTangent=1.1179, interpMode=RCIM_Cubic, leaveTangent=1.1179, tangentMode=RCTM_User, time=0.2042, value=0.6228}
          {arriveTangent=0.0224, interpMode=RCIM_Cubic, leaveTangent=0.0224, tangentMode=RCTM_User, time=1, value=1}
         ]
        }}})
    ))
  ]
  rendererClasses [&NiagaraMeshRendererProperties]
  renderers [{rendererClass=&NiagaraMeshRendererProperties, rendererIndex=0}]
 }
 BubbleLoop001 {
  emitterUpdateScript [
   &EmitterState()
   &SpawnRate(
    SpawnRate=&FloatFromCurve(
     FloatCurve=DataInterface() {propertyValues {Curve {keys [<at 0 1 RCIM_Cubic>, {arriveTangent=-7.1525, interpMode=RCIM_Cubic, leaveTangent=-7.1525, tangentMode=RCTM_User, time=0.175, value=0}]}}}
     CurveIndex=$Emitter.NormalizedLoopAge
     Scale Curve=12
    )
    Use Spawn Probability=false
   )
  ]
  particleSpawnScript [
   &InitializeParticle(
    Color=<rgba 0.564 1 0.612 0.5>
    Mesh Scale Mode=::NewEnumerator1
    Mesh Uniform Scale Max=0.07
    Mesh Uniform Scale Min=0.01
    Position Offset=$System.BurstPosition
    Sprite Size Mode=::NewEnumerator1
    Uniform Sprite Size Max=12
    Uniform Sprite Size Min=7
    UsePositionOffset=true
    Write Material Random=true
   )
   &ShapeLocation(MaskSpawnGroup=false, Override Seed=false, Sphere Radius=36, Sphere Surface Distribution=1, UseOwnerScale=true)
  ]
  particleUpdateScript [
   &PointAttractionForce(
    Attraction Strength=&RandomRangeFloat(Minimum=150, Maximum=225)
    Attractor Position:Vector3f=$System.BurstPosition
    Kill Radius=2
    Kill Within Radius=true
    Module.Use Falloff=true
    UseAttractorOffset=false
   )
   &SolveForcesAndVelocity(Clamp Velocity=false, Limit Acceleration=false)
   &ParticleState()
   &DynamicMaterialParameters(
    Index 3 Param 2=&MakeFloatFromVector(Vector=$Particles.Scale, Channel=::NewEnumerator0)
    Index 3 Param 4=&RandomRangeFloat(Minimum=0, Maximum=100)
    Module.Param0WriteEnabled3=false
    Module.Param1WriteEnabled3=true
    Module.Param2WriteEnabled3=false
    Module.Param3WriteEnabled3=true
    Write Parameter Index 0=false
    Write Parameter Index 3=true
   )
   &ScaleColor(
    Scale Alpha=&Multiply_Float(
     A=&FloatFromCurve(FloatCurve=DataInterface() {propertyValues {Curve {keys [
          {interpMode=RCIM_Cubic, tangentMode=RCTM_User, time=0, value=0}
          {interpMode=RCIM_Cubic, tangentMode=RCTM_User, time=0.3, value=1}
          {tangentMode=RCTM_User, time=1, value=0}
         ]}}})
     B=1
    )
    Scale RGB=&Multiply_VectorByFloat(
     Vector=<1 1 1>
     Float=&FloatFromCurve(FloatCurve=DataInterface() {propertyValues {Curve {keys [
          {interpMode=RCIM_Cubic, tangentMode=RCTM_User, time=0, value=0}
          {interpMode=RCIM_Cubic, tangentMode=RCTM_User, time=0.3, value=1}
          {tangentMode=RCTM_User, time=1, value=0}
         ]}}})
    )
    ScaleA=true
    ScaleRGB=true
   )
   &ScaleMeshSize(Scale Factor=&Multiply_VectorByFloat(
     Vector=<1 1 1>
     Float=&FloatFromCurve(FloatCurve=DataInterface() {propertyValues {Curve {keys [<at 0 0.5>,<at 1 1>]}}})
    ))
  ]
  rendererClasses [&NiagaraMeshRendererProperties]
  renderers [{rendererClass=&NiagaraMeshRendererProperties, rendererIndex=0}]
 }
 Bubbles {bEnabled=false, rendererClasses [], renderers []}
 Droplets {bEnabled=false, rendererClasses [], renderers []}
 Intro001 {
  emitterSpawnScript [
   &InitializeAudioPlayer(Audio Player=DataInterface() {propertyValues {
      Attenuation=null
      Concurrency=null
      MaxPlaysPerTick=10
      SoundToPlay=/CosmeticCompanions/Assets/Audio_Shared/Interaction/Petting/SC_Mimosa_Interaction_React_Default
      bAllowLoopingOneShotSounds=false
      bLimitPlaysPerTick=true
      bOnlyActiveDuringGameplay=false
      bStopWhenComponentIsDestroyed=true
      ConfigurationUserParameter {parameter {name=None, typeDefHandle {registeredTypeIndex=1}, varData []}}
      ParameterNames []
     }})
   /Niagara/Modules/Audio/PlayAudio() {module=PlayAudio001}
  ]
  emitterUpdateScript [
   &EmitterState()
   &SpawnRate(
    SpawnGroup=1
    SpawnRate=&FloatFromCurve(FloatCurve=DataInterface() {propertyValues {Curve {keys [<at 0.49 1>,<at 0.5 0>]}}}, CurveIndex=$Emitter.NormalizedLoopAge, Scale Curve=50)
    Use Spawn Probability=false
   ) {enabled=false}
   &SpawnBurst_Instantaneous(Use Spawn Probability=false, UseLoopCountLimit=false)
  ]
  particleSpawnScript [
   /SetVariables(Particles.VisibilityTag=$Emitter.SpawnGroup) {isSetParameters=true}
   &InitializeParticle(
    Color=<rgba 0 0.704 1 1>
    Lifetime=$System.IntroLength
    Position Offset=$System.BurstPosition
    Ribbon Width=10
    Ribbon Width Mode=::NewEnumerator1
    Sprite Rotation Mode=::NewEnumerator0
    Sprite Size Mode=::NewEnumerator3
    Uniform Sprite Size=&Multiply_Float(A=18, B=&FN_FrontEndFloatModifier(Gameplay=1, Front End=$User.FrontEndScale))
    UsePositionOffset=true
    Write Material Random=true
   )
  ]
  particleUpdateScript [
   &ParticleState()
   &ScaleSpriteSize(
    Scale Sprite Size Mode=::NewEnumerator0
    Uniform Scale Factor=&Add_Float(
     A=0
     B=&FloatFromCurve(FloatCurve=DataInterface() {propertyValues {Curve {
         defaultValue=0
         keys [
          {arriveTangent=1.4235, interpMode=RCIM_Cubic, leaveTangent=1.4235, tangentMode=RCTM_User, time=0, value=0}
          {arriveTangent=0.134, interpMode=RCIM_Cubic, leaveTangent=0.134, tangentMode=RCTM_User, time=0.8, value=1}
          {arriveTangent=1.6794, interpMode=RCIM_Cubic, leaveTangent=1.6794, tangentMode=RCTM_User, time=0.85, value=1}
          {arriveTangent=3.3106, interpMode=RCIM_Cubic, leaveTangent=3.3106, tangentMode=RCTM_User, time=1, value=2}
         ]
        }}})
    )
   ) {module=ScaleSpriteSize002}
   &ScaleColor(
    Scale Alpha=&FloatFromCurve(FloatCurve=DataInterface() {propertyValues {Curve {keys [<at 0 0 RCIM_Cubic>, {arriveTangent=1.779, interpMode=RCIM_Cubic, leaveTangent=1.779, tangentMode=RCTM_User, time=1, value=1}]}}})
    ScaleA=true
    ScaleRGB=true
   )
   &OffsetPosition(Position Offset=&Lerp_Vector(
     A=$System.HeartIntroSpawnPosition
     B=$System.BurstPosition
     Alpha=&FloatFromCurve(FloatCurve=DataInterface() {propertyValues {Curve {keys [
          {arriveTangent=3.0606, interpMode=RCIM_Cubic, leaveTangent=3.0606, tangentMode=RCTM_User, time=0, value=0}
          {arriveTangent=-0.1074, tangentMode=RCTM_User, time=1, value=1}
         ]}}})
    )) {enabled=false}
   &DynamicMaterialParameters(
    Index 0 Param 1=&FloatFromCurve(
     FloatCurve=DataInterface() {propertyValues {Curve {
        defaultValue=1
        keys [
         {arriveTangent=1.8779, interpMode=RCIM_Cubic, leaveTangent=1.8779, tangentMode=RCTM_User, time=0, value=0}
         {interpMode=RCIM_Cubic, leaveTangent=2.7455, time=1, value=1}
        ]
       }}}
     Scale Curve=2
    )
    Index 0 Param 2=&FloatFromCurve(
     FloatCurve=DataInterface() {propertyValues {Curve {
        defaultValue=1
        keys [
         {arriveTangent=2.1129, interpMode=RCIM_Cubic, leaveTangent=2.1129, tangentMode=RCTM_User, time=0, value=0}
         {interpMode=RCIM_Cubic, leaveTangent=2.7455, time=1, value=1}
        ]
       }}}
     Scale Curve=0.4
    )
    Index 0 Param 3=&FloatFromCurve(FloatCurve=DataInterface() {propertyValues {Curve {defaultValue=1, keys [<at 0 0 RCIM_Cubic>, {interpMode=RCIM_Cubic, leaveTangent=2.7455, time=1, value=1}]}}})
    Module.Index0Float0Write=true
    Module.Index0Float1Write=true
    Module.Index0Float2Write=true
    Module.Index0Float3Write=true
   )
   /SetVariables(
    Particles.Glow Color=&Multiply_LinearColorByFloat(Linear Color=<rgba 0.0879 0.4427 0.1073 0.5>, Float=0.3)
    Particles.Glow Size=&Multiply_Vector2D(A=$Particles.SpriteSize, B=<1.5 1.5>)
   ) {isSetParameters=true}
   &ScaleRibbonWidth(Ribbon Width Scale=&FloatFromCurve(FloatCurve=DataInterface() {propertyValues {Curve {keys [<at 0 1>,<at 0.2 1>,<at 0.7 1>,<at 1 0>]}}}))
  ]
  rendererClasses [&NiagaraSpriteRendererProperties]
  renderers [
   {rendererClass=&NiagaraSpriteRendererProperties}
   {rendererClass=&NiagaraSpriteRendererProperties, rendererIndex=1}
  ]
 }
 Minimal {
  emitterUpdateScript [
   &EmitterState()
   &SpawnParticlesFromOtherEmitter(
    Attribute Reader=DataInterface() {propertyValues {EmitterBinding {bindingMode=Other, emitterName=BubbleLoop001}}}
    Spawn Rate=120
    Use Maximum Age=false
    Use Minimum Age=false
   )
   /SetVariables(Emitter.TrailWidth:NiagaraFloat=&RandomRangeFloat(Minimum=0.8, Maximum=1.75)) {isSetParameters=true}
  ]
  particleSpawnScript [
   &InitializeParticle(Color=<rgba 0.564 1 0.612 0.5>, Lifetime=0.25, UsePositionOffset=false, Write Material Random=true)
   &SampleParticlesFromOtherEmitter(Particle ID Sampling=::NewEnumerator1)
  ]
  particleUpdateScript [
   &ParticleState()
   &SolveForcesAndVelocity(Clamp Velocity=false, Limit Acceleration=false)
  ]
  rendererClasses [&NiagaraRibbonRendererProperties]
  renderers [{rendererClass=&NiagaraRibbonRendererProperties}]
 }
 Minimal001 {
  emitterUpdateScript [
   &EmitterState()
   &SpawnBurst_Instantaneous(Spawn Count=8, Spawn Time=0.75, Use Spawn Probability=false, UseLoopCountLimit=false)
   &SpawnBurst_Instantaneous(Spawn Count=15, Spawn Time=0.75, Use Spawn Probability=false, UseLoopCountLimit=false) {module=SpawnBurst_Instantaneous001}
  ]
  particleSpawnScript [
   /SetVariables(Particles.VisibilityTag=&SpawnGroup()) {isSetParameters=true}
   &InitializeParticle(
    Color=&MaskLinearColorBySpawnGroup(
     Masked Linear Color=<rgba 1 1 1 0.5>
     Passthrough Linear Color=&RandomRangeLinearColor(Minimum=<rgba 0.624 1 0.7529 0.1255>, Maximum=<rgba 0.4793 0.9131 1 0.3255>)
    )
    Lifetime=&MaskFloatBySpawnGroup(
     Masked Float=&RandomRangeFloat(Minimum=0.35, Maximum=1.2)
     Passthrough Float=&RandomRangeFloat(Minimum=0.12, Maximum=0.5)
    )
    Mesh Scale Mode=::NewEnumerator1
    Mesh Uniform Scale Max=0.04
    Mesh Uniform Scale Min=0.007
    Position Offset=$System.BurstPosition
    Sprite Size Mode=::NewEnumerator1
    Uniform Sprite Size Max=5
    Uniform Sprite Size Min=2
    UsePositionOffset=true
    Write Material Random=true
   )
   &ShapeLocation(MaskSpawnGroup=false, Override Seed=false, Sphere Radius=6, UseOwnerScale=true)
   &AddVelocity(
    Module.Use Falloff Distance=false
    Origin Offset=$System.BurstPosition
    Override Seed=false
    Velocity Mode=::NewEnumerator1
    Velocity Speed=&MaskFloatBySpawnGroup(
     Masked Float=&RandomRangeFloat(Minimum=70, Maximum=150)
     Passthrough Float=&RandomRangeFloat(Minimum=30, Maximum=90)
    )
   )
  ]
  particleUpdateScript [
   &ParticleState()
   &ScaleSpriteSize(Uniform Curve Sprite Scale=DataInterface() {propertyValues {Curve {defaultValue=0, keys [<at 0 0.5>,<at 0.3555 1>,<at 1 0>]}}})
   &ScaleMeshSize(Scale Factor=&VectorFromFloat(Value=&FloatFromCurve(FloatCurve=DataInterface() {propertyValues {Curve {defaultValue=1, keys [<at 0 0>,<at 0.35 1>,<at 0.7 1>,<at 0.85 1.25>,<at 1 0.4>]}}})))
   &Drag(
    Drag=&MaskFloatBySpawnGroup(
     Masked Float=&RandomRangeFloat(Minimum=5, Maximum=8)
     Passthrough Float=&RandomRangeFloat(Minimum=1, Maximum=3)
    )
    Use Linear Drag=true
    Use Rotational Drag=true
   )
   &GravityForce(Gravity=&MakeVector(
     X=0
     Y=0
     Z=&MaskFloatBySpawnGroup(Masked Float=&RandomRangeFloat(Minimum=120, Maximum=225), Passthrough Float=0)
    ))
   &SolveForcesAndVelocity(Clamp Velocity=false, Limit Acceleration=false)
   &ScaleSpriteSizeBySpeed(
    Max Scale Factor=&MaskVector2dBySpawnGroup(Masked Vector2D=<1 1>, Passthrough Vector2D=<0.4 7>, Spawn Group Mask=0)
    Min Scale Factor=&MaskVector2dBySpawnGroup(Masked Vector2D=<1 1>, Passthrough Vector2D=<0.7 4>, Spawn Group Mask=0)
    Velocity Threshold=15
   )
   &ScaleColor(
    Scale Alpha=&MaskFloatBySpawnGroup(
     Masked Float=&FloatFromCurve(FloatCurve=DataInterface() {propertyValues {Curve {keys [<at 0 0 RCIM_Cubic>,<at 0.3 1 RCIM_Cubic>,<at 1 0 RCIM_Cubic>]}}})
     Passthrough Float=1
    )
    ScaleA=true
    ScaleRGB=true
   )
   /SetVariables(Particles.BubbleDP3:Vector4f=&MakeVector4(X=0, Y=3, Z=0, W=&RandomRangeFloat(Minimum=0, Maximum=100))) {isSetParameters=true}
  ]
  rendererClasses [&NiagaraMeshRendererProperties, &NiagaraSpriteRendererProperties]
  renderers [
   {rendererClass=&NiagaraMeshRendererProperties, rendererIndex=0}
   {rendererClass=&NiagaraSpriteRendererProperties, rendererIndex=1}
  ]
 }
 NewPop {
  emitterUpdateScript [
   &EmitterState()
   &SpawnBurst_Instantaneous(Spawn Time=&Subtract_Float(A=$System.IntroLength, B=0), Use Spawn Probability=false, UseLoopCountLimit=false)
  ]
  particleSpawnScript [&InitializeParticle(
    Lifetime=0.2
    Position Offset=$System.BurstPosition
    Sprite Rotation Mode=::NewEnumerator0
    Sprite Size Mode=::NewEnumerator3
    Uniform Sprite Size=&Multiply_Float(A=8, B=1)
    UsePositionOffset=true
    Write Material Random=true
   )]
  particleUpdateScript [
   &ParticleState()
   &ScaleSpriteSize(
    Scale Sprite Size Mode=::NewEnumerator0
    Uniform Scale Factor=&FloatFromCurve(FloatCurve=DataInterface() {propertyValues {Curve {keys [
         <at 0 2 RCIM_Cubic>
         {arriveTangent=-2.1754, interpMode=RCIM_Cubic, leaveTangent=-2.1754, time=0.175, value=3}
         {arriveTangent=-0.2182, interpMode=RCIM_Cubic, leaveTangent=3.345, tangentMode=RCTM_Break, time=0.2, value=1.5649}
         {arriveTangent=1.094, interpMode=RCIM_Cubic, leaveTangent=1.6744, tangentMode=RCTM_Break, time=0.9989, value=3.0533}
        ]}}})
   ) {module=ScaleSpriteSize002}
   &SolveForcesAndVelocity(Clamp Velocity=false, Limit Acceleration=false)
   &DynamicMaterialParameters(
    Index 0 Param 1=&FloatFromCurve(FloatCurve=DataInterface() {propertyValues {Curve {
        defaultValue=0
        keys [
         {arriveTangent=-4.0077, interpMode=RCIM_Cubic, leaveTangent=-4.0077, tangentMode=RCTM_User, time=0.0275, value=1}
         {arriveTangent=-0.0705, interpMode=RCIM_Cubic, leaveTangent=-0.0705, tangentMode=RCTM_User, time=0.6, value=0}
        ]
       }}})
    Index 0 Param 2=&FloatFromCurve(FloatCurve=DataInterface() {propertyValues {Curve {
        defaultValue=0
        keys [
         {arriveTangent=-4.0077, interpMode=RCIM_Cubic, leaveTangent=-4.0077, tangentMode=RCTM_User, time=0.0275, value=1}
         {arriveTangent=-0.0705, interpMode=RCIM_Cubic, leaveTangent=-0.0705, tangentMode=RCTM_User, time=0.6, value=0}
        ]
       }}})
    Index 0 Param 3=&FloatFromCurve(
     FloatCurve=DataInterface() {propertyValues {Curve {
        defaultValue=0
        keys [
         {arriveTangent=0.3137, interpMode=RCIM_Cubic, leaveTangent=0.3137, tangentMode=RCTM_User, time=0, value=1}
         {arriveTangent=-0.5451, interpMode=RCIM_Cubic, leaveTangent=-0.5451, tangentMode=RCTM_User, time=1, value=0}
        ]
       }}}
     Scale Curve=0
    )
    Index 0 Param 4=&FloatFromCurve(
     FloatCurve=DataInterface() {propertyValues {Curve {
        defaultValue=0
        keys [
         {arriveTangent=0.3137, interpMode=RCIM_Cubic, leaveTangent=0.3137, tangentMode=RCTM_User, time=0, value=1}
         {arriveTangent=-0.5451, interpMode=RCIM_Cubic, leaveTangent=-0.5451, tangentMode=RCTM_User, time=1, value=0}
        ]
       }}}
     Scale Curve=0
    )
    Module.Index0Float0Write=true
    Module.Index0Float1Write=true
    Module.Index0Float2Write=true
    Module.Index0Float3Write=true
   )
   /SetVariables(
    Particles.Flare Color:LinearColor=&MakeLinearColorFromVectorAndFloat(
     "Vector (RGB)"=&MakeVectorFromLinearColorRGB(Color=&FNTimeOfDayColor(Night Color=<rgba 0.8021 0.8021 0.8021 1>, Morning Color=<rgba 1 1 1 1>, Day Color=<rgba 3 3 3 1>, Evening Color=<rgba 1 1 1 1>))
     "Float (Alpha)"=&FloatFromCurve(
      FloatCurve=DataInterface() {propertyValues {Curve {
         defaultValue=1
         keys [
          {arriveTangent=9.1715, interpMode=RCIM_Cubic, leaveTangent=9.1715, tangentMode=RCTM_User, time=0, value=1}
          {interpMode=RCIM_Cubic, leaveTangent=-1, time=0.191, value=0}
         ]
        }}}
      Scale Curve=0.3
     )
    )
    Particles.Flare Size:Vector2f=&Vector2DFromFloat(Value=&FloatFromCurve(
      FloatCurve=DataInterface() {propertyValues {Curve {keys [{arriveTangent=1.01, interpMode=RCIM_Cubic, leaveTangent=1.01, tangentMode=RCTM_User, time=0, value=0.5}, <at 1 1 RCIM_Cubic>]}}}
      Scale Curve=92.2981
     ))
   ) {isSetParameters=true}
   /SetVariables(
    Particles.Glow Color=&Multiply_LinearColorByFloat(
     Linear Color=&MultiplyLinearColor(
      Linear Color=$Particles.Color
      Scale Factor=&MakeVector4_FromLinearColor(LinearColor=&FNTimeOfDayColor(
        Night Color=<rgba 0.4225 0.9219 0.4659 0.7>
        Morning Color=<rgba 0.3938 1.2 0.4639 0.7>
        Day Color=<rgba 0.7083 2 0.8206 0.7>
        Evening Color=<rgba 0.425 1.2 0.4924 0.7>
       ))
     )
     Float=0.225
    )
    Particles.Glow Size=&Vector2DFromFloat(Value=&Multiply_Float(
      A=1
      B=&Multiply_Float(
       A=35
       B=&FloatFromCurve(FloatCurve=DataInterface() {propertyValues {Curve {
           defaultValue=0
           keys [
            {arriveTangent=3.6099, interpMode=RCIM_Cubic, leaveTangent=3.6099, tangentMode=RCTM_User, time=0, value=0}
            {arriveTangent=0.0224, interpMode=RCIM_Cubic, leaveTangent=0.0224, tangentMode=RCTM_User, time=1, value=1}
           ]
          }}})
      )
     ))
   ) {isSetParameters=true}
   &ScaleColor(
    Scale Alpha=&FloatFromCurve(FloatCurve=DataInterface() {propertyValues {Curve {
        defaultValue=0
        keys [
         {arriveTangent=0.3137, interpMode=RCIM_Cubic, leaveTangent=0.3137, tangentMode=RCTM_User, time=0, value=1}
         {arriveTangent=-2.0958, interpMode=RCIM_Cubic, leaveTangent=-2.0958, tangentMode=RCTM_User, time=1, value=0}
        ]
       }}})
    ScaleA=true
    ScaleRGB=true
   )
  ]
  rendererClasses [&NiagaraSpriteRendererProperties]
  renderers [
   {rendererClass=&NiagaraSpriteRendererProperties}
   {rendererClass=&NiagaraSpriteRendererProperties, rendererIndex=1}
   {rendererClass=&NiagaraSpriteRendererProperties, rendererIndex=2}
  ]
 }
}
summary {
 systemName=NS_ReactFX_HeadWhale
 emitters [
  {emitterName=BubbleLoop001, rendererClasses [&NiagaraMeshRendererProperties]}
  {emitterName=Minimal, rendererClasses [&NiagaraRibbonRendererProperties]}
  {emitterName=Intro001, rendererClasses [&NiagaraSpriteRendererProperties]}
  {emitterName=BubbleLoop, rendererClasses [&NiagaraMeshRendererProperties]}
  {emitterName=NewPop, rendererClasses [&NiagaraSpriteRendererProperties]}
  {bEnabled=false, emitterName=Bubbles, rendererClasses []}
  {bEnabled=false, emitterName=Droplets, rendererClasses []}
  {emitterName=Minimal001, rendererClasses [&NiagaraMeshRendererProperties, &NiagaraSpriteRendererProperties]}
 ]
 userVariables [
  User.bIsFrontEnd {type=NiagaraBool}
  User.bIsFrontEndPreview {type=NiagaraBool}
  User.FrontEndScale {type=NiagaraFloat, defaultValue=0.6, description=""}
 ]
}
