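# Niagara reaction effect for the HoppyBun cosmetic companion: a timed egg burst with a companion petting-reaction sound.
# The Egg emitter plays the sound and staggers sprite bursts, with SpawnGroup switches giving each group its own motion, scale and color within one emitter.
# Separate glitter and smoke emitters add delayed pink bursts, with smoke driven by dynamic material parameters.
# Time-of-day inputs tint particles, and front-end modifiers adjust offsets for menu display.
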NiagaraSystem /CosmeticCompanions/FX/React/NS_ReactFX_HoppyBun
dependencies {
 usedDataInterfaces [
  &NiagaraDataInterfaceCurve
  &NiagaraDataInterfaceVector2DCurve
  &NiagaraDataInterfaceVectorCurve
  &NiagaraDataInterfaceAudioPlayer
  &NiagaraDataInterfaceSpriteRendererInfo
  &NiagaraDataInterfaceArrayColor
 ]
 usedDynamicInputs []
 usedModules [
  &self:SystemSpawnScript.SystemScriptSource.SystemScriptGraph.NiagaraNodeAssignment_1.SetVariables_#1
  &self:SystemSpawnScript.SystemScriptSource.SystemScriptGraph.NiagaraNodeAssignment_2.SetVariables_#2
  &self:SystemSpawnScript.SystemScriptSource.SystemScriptGraph.NiagaraNodeAssignment_3.SetVariables_#3
  &self:SystemSpawnScript.SystemScriptSource.SystemScriptGraph.NiagaraNodeAssignment_4.SetVariables_#4
  &self:SystemSpawnScript.SystemScriptSource.SystemScriptGraph.NiagaraNodeAssignment_5.SetVariables_#5
  &self:SystemSpawnScript.SystemScriptSource.SystemScriptGraph.NiagaraNodeAssignment_11.SetVariables_#6
  &self:SystemSpawnScript.SystemScriptSource.SystemScriptGraph.NiagaraNodeAssignment_12.SetVariables_#7
  &self:SystemSpawnScript.SystemScriptSource.SystemScriptGraph.NiagaraNodeAssignment_13.SetVariables_#8
  &self:SystemSpawnScript.SystemScriptSource.SystemScriptGraph.NiagaraNodeAssignment_14.SetVariables_#9
  &SystemState
  &self:SystemSpawnScript.SystemScriptSource.SystemScriptGraph.NiagaraNodeAssignment_30.SetVariables_#10
 ]
}
emitters {
 Egg {
  emitterSpawnScript [
   &InitializeAudioPlayer(Audio Player=DataInterface() {propertyValues {
      Attenuation=null
      Concurrency=null
      MaxPlaysPerTick=10
      SoundToPlay=/CosmeticCompanions/Assets/Audio_Shared/Interaction/Petting/SC_Mimosa_Interaction_React_Default
      bAllowLoopingOneShotSounds=false
      bLimitPlaysPerTick=true
      bOnlyActiveDuringGameplay=false
      bStopWhenComponentIsDestroyed=true
      ConfigurationUserParameter {parameter {name=None, typeDefHandle {registeredTypeIndex=1}, varData []}}
      ParameterNames []
     }})
   /Niagara/Modules/Audio/PlayAudio() {module=PlayAudio001}
  ]
  emitterUpdateScript [
   &EmitterState()
   &SpawnBurst_Instantaneous(Spawn Time=0.1, Use Spawn Probability=false, UseLoopCountLimit=false)
   &SpawnBurst_Instantaneous(Spawn Group=1, Spawn Time=0.6, Use Spawn Probability=false, UseLoopCountLimit=false) {module=SpawnBurst_Instantaneous001}
   &SpawnBurst_Instantaneous(Spawn Count=8, Spawn Group=2, Spawn Time=0.53, Use Spawn Probability=false, UseLoopCountLimit=false) {module=SpawnBurst_Instantaneous002}
  ]
  particleSpawnScript [
   /SetVariables(Particles.VisibilityTag=&SpawnGroup()) {isSetParameters=true}
   &InitializeParticle(
    Color=&MergeHelper_ColorSwitch(
     0=&FNTimeOfDayColor(
      Night Color=<rgba 0.6117 0.7914 1 1>
      Morning Color=<rgba 0.9573 0.9067 0.8546 1>
      Day Color=<rgba 1 1 1 1>
      Evening Color=<rgba 0.9022 0.8146 0.7243 1>
     )
     1=&FNTimeOfDayColor(
      Night Color=<rgba 0.6117 0.7914 1 1>
      Morning Color=<rgba 0.9573 0.9067 0.8546 1>
      Day Color=<rgba 1 1 1 1>
      Evening Color=<rgba 0.9022 0.8146 0.7243 1>
     )
     2=&FNTimeOfDayColor(
      Night Color=<rgba 0.4271 0.0681 0.3751 1>
      Morning Color=<rgba 1 0.0594 0.8476 1>
      Day Color=<rgba 1 0.0052 0.8388 1>
      Evening Color=<rgba 1 0.1677 0.764 1>
     )
     3=&SelectLinearColorFromArray(
      Array Sampling Mode=::NewEnumerator1
      Direct Array Index=&RandomRangeInt(Minimum=0, Maximum=1)
      Color Selection Array=$System.EggColors
      Random Seed=?
     )
     Index=&SpawnGroup()
     Num=::NewEnumerator1
    )
    Lifetime=&MergeHelper_FloatSwitch(
     0=0.55
     1=1.05
     2=&RandomRangeFloat(Minimum=0.55, Maximum=0.3)
     3=&RandomRangeFloat(Minimum=0.5, Maximum=1)
     Index=&SpawnGroup()
     Num=::NewEnumerator1
    )
    Position Offset=&FN_FrontEndVector3Modifier(Gameplay=<0 0 0>, Front End=<0 0 0>)
    Sprite Rotation Angle=&MergeHelper_FloatSwitch(
     0=0
     1=0
     2=&RandomRangeFloat(Minimum=0, Maximum=360)
     3=?
     Index=&SpawnGroup()
     Num=::NewEnumerator1
    )
    Sprite Rotation Mode=::NewEnumerator0
    Sprite Size Mode=::NewEnumerator3
    Uniform Sprite Size=&MergeHelper_FloatSwitch(
     0=24
     1=1
     2=0
     3=&RandomRangeFloat(Minimum=20, Maximum=30)
     Index=&SpawnGroup()
     Num=::NewEnumerator1
    )
    UsePositionOffset=true
    Write Material Random=true
   ) {module=InitializeParticle001}
   &CameraOffset(Camera Offset Amount=15)
   &ShapeLocation(
    MaskSpawnGroup=false
    Offset=&FN_FrontEndVector3Modifier(Gameplay=<0 0 0>, Front End=$System.BurstPosition)
    Offset Mode=::NewEnumerator0
    Override Seed=false
    Sphere Radius=&MaskFloat(Mask Source=0, Mask Parameter=?, Mask=2, Masked Float=2, Passthrough Float=1)
    Sphere Surface Distribution=1
    UseOwnerScale=true
   )
   &AddVelocity(
    Module.Use Falloff Distance=false
    Origin Offset=&FN_FrontEndVector3Modifier(Gameplay=<0 0 0>, Front End=$System.BurstPosition)
    Override Seed=false
    Velocity Mode=::NewEnumerator1
    Velocity Speed=&MergeHelper_FloatSwitch(
     0=0
     1=0
     2=&RandomRangeFloat(Minimum=90, Maximum=150)
     3=?
     Index=&SpawnGroup()
     Num=::NewEnumerator1
    )
   ) {module=AddVelocity001}
   &AddVelocity(Velocity=&MergeHelper_VectorSwitch(
     0=<0 0 0>
     1=<0 0 0>
     2=&RandomRangeVector(Minimum=<0 0 40>, Maximum=<0 0 80>)
     3=?
     Index=&SpawnGroup()
     Num=::NewEnumerator1
    )) {module=AddVelocity002}
  ]
  particleUpdateScript [
   &ParticleState()
   &ScaleSpriteSize(Uniform Curve Sprite Scale=DataInterface() {propertyValues {Curve {
       defaultValue=0
       keys [
        {interpMode=RCIM_Cubic, tangentMode=RCTM_User, time=0, value=0}
        {interpMode=RCIM_Cubic, tangentMode=RCTM_User, time=0.2, value=1}
        {arriveTangent=2.494, interpMode=RCIM_Cubic, leaveTangent=2.494, time=0.87, value=1}
        <at 1.0019 3 RCIM_Cubic>
       ]
      }}})
   /SetVariables(Particles.ChocoScale:Vector2f=&MergeHelper_Vector2Switch(
     0=<0 0>
     1=&Vector2DFromCurve(
      Vector2Curve=DataInterface() {propertyValues {
        XCurve {
         defaultValue=1
         keys [
          {interpMode=RCIM_Cubic, tangentMode=RCTM_User, time=0, value=0}
          {interpMode=RCIM_Cubic, tangentMode=RCTM_User, time=0.25, value=1}
          {arriveTangent=-0.9333, interpMode=RCIM_Cubic, leaveTangent=-0.9333, time=0.92, value=1}
          <at 1 0.3 RCIM_Cubic>
         ]
        }
        YCurve {
         defaultValue=1
         keys [
          {interpMode=RCIM_Cubic, tangentMode=RCTM_User, time=0, value=0}
          {interpMode=RCIM_Cubic, tangentMode=RCTM_User, time=0.25, value=1}
          {arriveTangent=-0.9333, interpMode=RCIM_Cubic, leaveTangent=-0.9333, time=0.92, value=1}
          <at 1 0.3 RCIM_Cubic>
         ]
        }
       }}
      ScaleCurve=&Multiply_Vector2D_ByFloat(Float=28, Vector2D=<1 1>)
     )
     2=&Vector2DFromCurve(
      Vector2Curve=DataInterface() {propertyValues {
        XCurve {defaultValue=0, keys [{tangentMode=RCTM_User, time=0.75, value=1}, {tangentMode=RCTM_User, time=1, value=0}]}
        YCurve {defaultValue=0, keys [{tangentMode=RCTM_User, time=0.75, value=1}, {tangentMode=RCTM_User, time=1, value=0}]}
       }}
      ScaleCurve=&Multiply_Vector2D_ByFloat(Float=&RandomRangeFloat(Minimum=4, Maximum=8.5), Vector2D=<1 1>)
     )
     3=&Vector2DFromCurve(
      Vector2Curve=DataInterface() {propertyValues {
        XCurve {
         defaultValue=1
         keys [{interpMode=RCIM_Cubic, tangentMode=RCTM_User, time=0.6, value=1}, {interpMode=RCIM_Cubic, tangentMode=RCTM_User, time=1, value=0}]
        }
        YCurve {
         defaultValue=1
         keys [{interpMode=RCIM_Cubic, tangentMode=RCTM_User, time=0.6, value=1}, {interpMode=RCIM_Cubic, tangentMode=RCTM_User, time=1, value=0}]
        }
       }}
      ScaleCurve=<1 1>
     )
     4=?
     Index=&SpawnGroup()
     Num=::NewEnumerator1
    )) {isSetParameters=true}
   &ScaleColor(
    Scale Alpha=&MergeHelper_FloatSwitch(
     0=&FloatFromCurve(FloatCurve=DataInterface() {propertyValues {Curve {
         defaultValue=1
         keys [{interpMode=RCIM_Cubic, tangentMode=RCTM_User, time=0.85, value=1}, {interpMode=RCIM_Cubic, tangentMode=RCTM_User, time=1, value=0}]
        }}})
     1=1
     2=&FloatFromCurve(FloatCurve=DataInterface() {propertyValues {Curve {defaultValue=0, keys [<at 0 0>,<at 0.1364 0.5>,<at 0.874 0.5>,<at 1 0>]}}})
     3=&FloatFromCurve(FloatCurve=DataInterface() {propertyValues {Curve {keys [<at 0 1>,<at 1 0>]}}})
     Index=&SpawnGroup()
     Num=::NewEnumerator1
    )
    Scale RGB=&MergeHelper_VectorSwitch(
     0=<1 1 1>
     1=<1 1 1>
     2=&Multiply_VectorByFloat(Vector=<1 1 1>, Float=&RandomRangeFloat(Minimum=1, Maximum=4))
     3=&Multiply_VectorByFloat(Vector=<1 1 1>, Float=&RandomRangeFloat(Minimum=1, Maximum=5))
     Index=&SpawnGroup()
     Num=::NewEnumerator1
    )
    ScaleA=true
    ScaleRGB=true
   ) {module=ScaleColor001}
   &Drag(
    Drag=&MergeHelper_FloatSwitch(
     0=0
     1=0
     2=&RandomRangeFloat(Minimum=6, Maximum=4.5)
     3=?
     Index=&SpawnGroup()
     Num=::NewEnumerator1
    )
    Use Linear Drag=true
    Use Rotational Drag=true
   )
   &GravityForce(Gravity=&MergeHelper_VectorSwitch(
     0=<0 0 0>
     1=<0 0 0>
     2=<0 0 -180>
     3=&RandomRangeVector(Minimum=<0 0 -400>, Maximum=<0 0 -200>)
     Index=&SpawnGroup()
     Num=::NewEnumerator1
    ))
   &CurlNoiseForce(
    Bias Noise=false
    Noise Frequency=&MergeHelper_FloatSwitch(0=0, 1=0, 2=0, 3=100, Index=&SpawnGroup(), Num=::NewEnumerator2)
    Noise Strength=&MergeHelper_FloatSwitch(0=0, 1=0, 2=0, 3=100, Index=&SpawnGroup(), Num=::NewEnumerator2)
    Pan Noise=false
   ) {enabled=false}
   &SolveForcesAndVelocity(Clamp Velocity=false, Limit Acceleration=false)
   &SpriteRotationRate(Rotation Rate=&MergeHelper_FloatSwitch(
     0=0
     1=0
     2=0
     3=&RandomRangeFloat(Minimum=-5, Maximum=5)
     Index=&SpawnGroup()
     Num=::NewEnumerator1
    ))
   &SubUVAnimation(
    Evaluation Type=::NewEnumerator0
    "Sprite Renderer "=DataInterface() {propertyValues {SpriteRenderer=&self:Egg_0.NiagaraSpriteRendererProperties_3}}
    SubUV Animation Mode:ENiagara_SubUVLookupModeV2=::NewEnumerator2
    UseEndFrame=false
    UseStartFrame=false
   )
  ]
  rendererClasses [&NiagaraSpriteRendererProperties]
  renderers [
   {rendererClass=&NiagaraSpriteRendererProperties}
   {rendererClass=&NiagaraSpriteRendererProperties, rendererIndex=1}
   {rendererClass=&NiagaraSpriteRendererProperties, rendererIndex=2}
  ]
 }
 EggBurst_Glitter {
  emitterUpdateScript [
   &EmitterState()
   &SpawnBurst_Instantaneous(Spawn Count=8, Spawn Time=1.4, Use Spawn Probability=false, UseLoopCountLimit=false) {module=SpawnBurst_Instantaneous001}
  ]
  particleSpawnScript [
   &InitializeParticle(
    Color=&RandomRangeLinearColor(
     Minimum=&FNTimeOfDayColor(
      Night Color=<rgba 0.4271 0.0681 0.3751 1>
      Morning Color=<rgba 1 0.0594 0.8476 1>
      Day Color=<rgba 1 0.0052 0.8388 1>
      Evening Color=<rgba 1 0.1677 0.764 1>
     )
     Maximum=&FNTimeOfDayColor(
      Night Color=<rgba 0.7406 0.1335 0.6527 1>
      Morning Color=<rgba 1 0.301 0.8867 1>
      Day Color=<rgba 1 0.3844 0.9002 1>
      Evening Color=<rgba 0.6211 0.1773 1 1>
     )
    )
    Lifetime=&RandomRangeFloat(Minimum=0.5, Maximum=1.5)
    Position Offset=&FN_FrontEndVector3Modifier(Gameplay=<0 0 -15>, Front End=<0 0 -40>)
    Sprite Size Mode=::NewEnumerator3
    Uniform Sprite Size=&RandomRangeFloat(Minimum=7, Maximum=10)
    UsePositionOffset=true
    Write Material Random=true
   ) {module=InitializeParticle001}
   &ShapeLocation(MaskSpawnGroup=false, Offset Mode=::NewEnumerator0, Override Seed=false, Sphere Radius=2, UseOwnerScale=true)
   &AddVelocity(Velocity=&RandomRangeVector(Minimum=<-50 -40 110>, Maximum=<50 40 200>))
  ]
  particleUpdateScript [
   &GravityForce(Gravity=&RandomRangeVector(Minimum=<0 0 -400>, Maximum=<0 0 -200>))
   &Drag(Rotational Drag=0, Use Linear Drag=true, Use Rotational Drag=true)
   &CurlNoiseForce(Bias Noise=false, Noise Frequency=100, Noise Strength=100, Pan Noise=false)
   &SolveForcesAndVelocity(Clamp Velocity=false, Limit Acceleration=false)
   &ParticleState()
   &ScaleSpriteSize(
    Uniform Curve Scale=&Multiply_Float(
     A=&RandomRangeFloat(Minimum=3, Maximum=7)
     B=&RandomRangeFloat(Minimum=0.5, Maximum=1, Evaluation Type=::NewEnumerator1)
    )
    Uniform Curve Sprite Scale=DataInterface() {propertyValues {Curve {keys [
        <at 0 0 RCIM_Cubic>
        {arriveTangent=-2.9089, interpMode=RCIM_Cubic, leaveTangent=-2.9089, tangentMode=RCTM_User, time=0.2, value=1}
        <at 1 0 RCIM_Cubic>
       ]}}}
   )
   &SpriteRotationRate(Rotation Rate=&RandomRangeFloat(Minimum=-5, Maximum=5))
   &ScaleColor(
    Scale RGB=&Multiply_VectorByFloat(Vector=<1 1 1>, Float=&RandomRangeFloat(Minimum=1, Maximum=4))
    ScaleA=true
    ScaleRGB=true
   )
  ]
  rendererClasses [&NiagaraSpriteRendererProperties]
  renderers [{rendererClass=&NiagaraSpriteRendererProperties}]
 }
 EggBurst_SmokeCloud002 {
  emitterUpdateScript [
   &EmitterState()
   &SpawnBurst_Instantaneous(Spawn Count=2, Spawn Time=1.55, Use Spawn Probability=false, UseLoopCountLimit=false) {module=SpawnBurst_Instantaneous001}
   &SpawnBurst_Instantaneous(Spawn Group=1, Use Spawn Probability=false, UseLoopCountLimit=false)
  ]
  particleSpawnScript [
   /SetVariables(Particles.VisibilityTag=&SpawnGroup()) {isSetParameters=true}
   &InitializeParticle(
    Color=&MergeHelper_ColorSwitch(
     0=&FNTimeOfDayColor(
      Night Color=<rgba 1 0.3371 0.9566 1>
      Morning Color=<rgba 1 0.3302 0.7883 1>
      Day Color=<rgba 1 0.2423 0.7605 1>
      Evening Color=<rgba 1 0.157 0.9222 1>
     )
     1=&FNTimeOfDayColor(
      Night Color=<rgba 0.2149 0.3102 0.842 1>
      Morning Color=<rgba 0.7078 0.5363 0.5148 1>
      Day Color=<rgba 0.8096 0.8096 0.8096 1>
      Evening Color=<rgba 0.6076 0.4758 0.7031 1>
     )
     2=?
     3=?
     Index=&SpawnGroup()
     Num=::NewEnumerator0
    )
    Lifetime=&MergeHelper_FloatSwitch(
     0=&RandomRangeFloat(Minimum=0.21, Maximum=0.35)
     1=&RandomRangeFloat(Minimum=0.3, Maximum=0.35)
     2=?
     3=?
     Index=&SpawnGroup()
     Num=::NewEnumerator0
    )
    Position Offset=&MergeHelper_VectorSwitch(
     0=&FN_FrontEndVector3Modifier(Gameplay=<0 0 -13>, Front End=<0 0 -20>)
     1=&FN_FrontEndVector3Modifier(Gameplay=<0 0 0>, Front End=<0 0 -20>)
     2=?
     3=?
     Index=&SpawnGroup()
     Num=::NewEnumerator0
    )
    Sprite Size Mode=::NewEnumerator1
    Sprite UV Mode=::NewEnumerator1
    Uniform Sprite Size Max=52
    Uniform Sprite Size Min=30
    UsePositionOffset=true
    Write Material Random=true
   ) {module=InitializeParticle001}
   &ApplyInitialForces()
   &AddVelocity(Velocity=&MergeHelper_VectorSwitch(
     0=&RandomRangeVector(Minimum=<0 0 70>, Maximum=<0 0 100>)
     1=<0 0 0>
     2=?
     3=?
     Index=&SpawnGroup()
     Num=::NewEnumerator0
    )) {module=AddVelocity001}
   &CameraOffset(Camera Offset Amount=&MergeHelper_FloatSwitch(
     0=&RandomRangeFloat(Minimum=1, Maximum=7)
     1=20
     2=?
     3=?
     Index=&SpawnGroup()
     Num=::NewEnumerator0
    ))
   /SetVariables(Particles.DMP_DustChunks:Vector4f=/Game/Effects/Niagara/Modules/MergeEmitters/MergeHelper_Vector4Switch(
     0:Vector4f=<1 1 1 1>
     1:Vector4f=&MakeVector4(X=0, Y=&UniformAOrBFloat(A=0), Z=0, W=0)
     2:Vector4f=?
     3:Vector4f=?
     4:Vector4f=?
     5:Vector4f=?
     6:Vector4f=?
     7:Vector4f=?
     8:Vector4f=?
     9:Vector4f=?
     Index:NiagaraInt32=&SpawnGroup()
     Num:MergeHelper_SwitcherEnum=::NewEnumerator0
    )) {isSetParameters=true}
  ]
  particleUpdateScript [
   &ParticleState()
   &ScaleSpriteSize(
    Non-Uniform Curve Scale=<1.2 1.2>
    Non-Uniform Curve Sprite Scale=DataInterface() {propertyValues {
      XCurve {keys [
        {arriveTangent=2.2927, interpMode=RCIM_Cubic, leaveTangent=2.2927, tangentMode=RCTM_User, time=0, value=0}
        {arriveTangent=0.0224, interpMode=RCIM_Cubic, leaveTangent=0.0224, tangentMode=RCTM_User, time=1, value=1}
       ]}
      YCurve {keys [
        {arriveTangent=2.2927, interpMode=RCIM_Cubic, leaveTangent=2.2927, tangentMode=RCTM_User, time=0, value=0}
        {arriveTangent=0.0224, interpMode=RCIM_Cubic, leaveTangent=0.0224, tangentMode=RCTM_User, time=1, value=1}
       ]}
     }}
    Scale Sprite Size Mode=::NewEnumerator3
   )
   &AccelerationForce(Acceleration=&MergeHelper_VectorSwitch(
     0=&VectorFromCurve(
      VectorCurve=DataInterface() {propertyValues {
        XCurve {keys [<at 0 0>,<at 1 0>]}
        YCurve {keys [{tangentMode=RCTM_User, time=0, value=1}, {tangentMode=RCTM_User, time=1, value=0}]}
        ZCurve {keys [
          {arriveTangent=-2.1026, interpMode=RCIM_Cubic, leaveTangent=-2.1026, tangentMode=RCTM_User, time=0, value=1}
          {interpMode=RCIM_Cubic, tangentMode=RCTM_User, time=1, value=0}
         ]}
       }}
      Scale Curve=&RandomRangeVector(Minimum=<0 0 0>, Maximum=<0 0 200>)
     )
     1=<0 0 0>
     2=?
     3=?
     Index=&SpawnGroup()
     Num=::NewEnumerator0
    )) {module=AccelerationForce001}
   /SetVariables(
    Particles.DMP_DustChunks:Vector4f=&MergeHelper_Vector4Switch(
     0:Vector4f=&MakeVector4(
      X=&RandomRangeFloat(Minimum=0, Maximum=1)
      Y=1
      Z=&RandomRangeFloat(Minimum=0.2, Maximum=0.3)
      W=&FloatFromCurve(
       FloatCurve=DataInterface() {propertyValues {Curve {keys [{tangentMode=RCTM_User, time=0, value=0}, {tangentMode=RCTM_User, time=1, value=1}]}}}
       Scale Curve=1.2
      )
     )
     1:Vector4f=&MakeVector4(
      X=&FloatFromCurve(FloatCurve=DataInterface() {propertyValues {Curve {keys [<at 0 0>,<at 1 1>]}}})
      Y=1
      Z=0
      W=0
     )
     2:Vector4f=?
     3:Vector4f=?
     4:Vector4f=?
     5:Vector4f=?
     6:Vector4f=?
     7:Vector4f=?
     8:Vector4f=?
     9:Vector4f=?
     Index:NiagaraInt32=&SpawnGroup()
     Num:MergeHelper_SwitcherEnum=::NewEnumerator0
    )
    Particles.DustChunksColor:LinearColor=&MergeHelper_ColorSwitch(
     0=&Multiply_LinearColorByFloat(
      Linear Color=&Multiply_LinearColorByFloat(
       Linear Color=&MultiplyLinearColor(
        Linear Color=$Particles.Initial.Color
        Scale Factor=/Niagara/DynamicInputs/UniformRange/V2/RandomRangeVector4(
         Minimum:Vector4f=<1 1 1 0>
         Maximum:Vector4f=<2 2 2 1>
         Evaluation Type:ENiagaraRandomnessEvaluation=::NewEnumerator0
         Recalculate Random Each Loop:NiagaraBool=true
         Randomness Mode:ENiagaraRandomnessMode=::NewEnumerator0
         Random Seed:NiagaraInt32=0
         Fixed Random Seed:NiagaraBool=false
        )
       )
       Float=&RandomRangeFloat(Minimum=0.7, Maximum=1)
      )
      Float=$System.ToD
     )
     1=&Multiply_LinearColorByFloat(
      Linear Color=&Multiply_LinearColorByFloat(
       Linear Color=<rgba 1 1 1 1>
       Float=&FloatFromCurve(FloatCurve=DataInterface() {propertyValues {Curve {keys [<at 0 1>,<at 0.8091 0.9932>,<at 1 0>]}}})
      )
      Float=$System.ToD
     )
     2=?
     3=?
     Index=&SpawnGroup()
     Num=::NewEnumerator0
    )
   ) {isSetParameters=true}
   &ScaleColor(
    Scale Alpha=&FloatFromCurve(
     FloatCurve=DataInterface() {propertyValues {Curve {keys [<at 0 2 RCIM_Cubic>, {tangentMode=RCTM_User, time=1, value=0}]}}}
     Scale Curve=&RandomRangeFloat(Minimum=0.8, Maximum=1)
    )
    Scale RGB=&VectorFromFloat(Value=&RandomRangeFloat(Minimum=1, Maximum=5))
    ScaleA=true
    ScaleRGB=true
   )
   &ScaleColor(
    Scale Alpha=&FloatFromCurve(
     FloatCurve=DataInterface() {propertyValues {Curve {keys [
         {arriveTangent=-0.0888, interpMode=RCIM_Cubic, leaveTangent=-0.0888, tangentMode=RCTM_User, time=0, value=0}
         {arriveTangent=0.12, interpMode=RCIM_Cubic, leaveTangent=0.12, tangentMode=RCTM_User, time=0.15, value=0.9847}
         {arriveTangent=-2.7131, interpMode=RCIM_Cubic, leaveTangent=-2.7131, tangentMode=RCTM_User, time=1, value=0}
        ]}}}
     Scale Curve=&Multiply_Float(A=1, B=$System.ToD)
    )
    Scale RGB=&VectorFromFloat(Value=&Multiply_Float(A=1, B=$System.ToD))
    ScaleA=true
    ScaleRGB=true
   ) {module=ScaleColor003}
   &GravityForce(Gravity=&MergeHelper_VectorSwitch(
     0=&VectorFromCurve(
      VectorCurve=DataInterface() {propertyValues {
        XCurve {keys [<at 0 0>,<at 1 0>]}
        YCurve {keys [<at 0 0>,<at 1 0>]}
        ZCurve {keys [
          {arriveTangent=1.0661, interpMode=RCIM_Cubic, leaveTangent=1.0661, tangentMode=RCTM_User, time=0, value=0}
          {interpMode=RCIM_Cubic, tangentMode=RCTM_User, time=1, value=1}
         ]}
       }}
      Scale Curve=&RandomRangeVector(Minimum=<0 0 -300>, Maximum=<0 0 -600>)
     )
     1=<0 0 0>
     2=?
     3=?
     Index=&SpawnGroup()
     Num=::NewEnumerator0
    ))
   &Drag(
    Drag=&MergeHelper_FloatSwitch(
     0=&FloatFromCurve(FloatCurve=DataInterface() {propertyValues {Curve {keys [
          {interpMode=RCIM_Cubic, tangentMode=RCTM_User, time=0, value=0}
          {interpMode=RCIM_Cubic, tangentMode=RCTM_User, time=0.7, value=0.9}
          {tangentMode=RCTM_User, time=1, value=0}
         ]}}})
     1=0
     2=?
     3=?
     Index=&SpawnGroup()
     Num=::NewEnumerator0
    )
    Use Linear Drag=true
    Use Rotational Drag=true
   ) {module=Drag002}
   &SolveForcesAndVelocity(Clamp Velocity=false, Limit Acceleration=false)
   &DynamicMaterialParameters(
    Index 0 Param 1=&Add_Float(
     A=&RandomRangeFloat(Minimum=0, Maximum=1)
     B=&FloatFromCurve(
      FloatCurve=DataInterface() {propertyValues {Curve {keys [
          {arriveTangent=2.7533, interpMode=RCIM_Cubic, leaveTangent=1.8958, tangentMode=RCTM_User, time=0, value=1}
          {arriveTangent=0.8292, interpMode=RCIM_Cubic, leaveTangent=0.8292, tangentMode=RCTM_User, time=0.321, value=1.3453}
          {arriveTangent=0.0529, interpMode=RCIM_Cubic, leaveTangent=0.2021, tangentMode=RCTM_User, time=1, value=1.5}
         ]}}}
      CurveIndex=&Add_Float(A=&Multiply_Float(A=1, B=0.5), B=$Particles.NormalizedAge)
     )
    )
    Index 0 Param 2=&FloatFromCurve(FloatCurve=DataInterface() {propertyValues {Curve {keys [
         {arriveTangent=0.0453, interpMode=RCIM_Cubic, tangentMode=RCTM_Break, time=0, value=-1}
         {arriveTangent=0.0604, interpMode=RCIM_Cubic, leaveTangent=0.0604, tangentMode=RCTM_User, time=1, value=-1}
        ]}}})
    Index 0 Param 3=0.5
    Index 0 Param 4=&FloatFromCurve(
     FloatCurve=DataInterface() {propertyValues {Curve {keys [<at 0.5 1>,<at 1 0>]}}}
     Scale Curve=&FNTimeOfDayScalar(
      Night Scale=2
      Morning Scale=2
      Day Scale=&FN_FrontEndFloatModifier(Gameplay=1, Front End=0.5)
      Evening Scale=3
     )
    )
    Module.Index0Float0Write=true
    Module.Index0Float1Write=true
    Module.Index0Float2Write=true
    Module.Index0Float3Write=true
   )
  ]
  rendererClasses [&NiagaraSpriteRendererProperties]
  renderers [
   {rendererClass=&NiagaraSpriteRendererProperties}
   {rendererClass=&NiagaraSpriteRendererProperties, rendererIndex=1}
   {rendererClass=&NiagaraSpriteRendererProperties, rendererIndex=2}
  ]
 }
 LWE_AirGlow {rendererClasses [], renderers []}
 LWE_Rays {rendererClasses [], renderers []}
}
summary {
 systemName=NS_ReactFX_HoppyBun
 emitters [
  {emitterName=LWE_AirGlow, rendererClasses []}
  {emitterName=LWE_Rays, rendererClasses []}
  {emitterName=EggBurst_SmokeCloud002, rendererClasses [&NiagaraSpriteRendererProperties]}
  {emitterName=EggBurst_Glitter, rendererClasses [&NiagaraSpriteRendererProperties]}
  {emitterName=Egg, rendererClasses [&NiagaraSpriteRendererProperties]}
 ]
 userVariables [
  User.bIsFrontEnd {type=NiagaraBool}
  User.bIsFrontEndPreview {type=NiagaraBool}
 ]
}
