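# Niagara reaction effect for the IcyScarf cosmetic companion: a heart with a snowflake rises to a burst point and pops in icy shards, with a reaction sound.
# BubbleHeart plays the shared petting-interaction audio and lerps between system intro and burst positions; Snow_Ribbon spawns trails from its particles.
# SpawnBurst swirls sparks around the intro position via vortex force.
# Pop uses delayed spawn groups for sprite/mesh debris, airglow, rays and corona, with time-of-day tinting.
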NiagaraSystem /CosmeticCompanions/FX/React/NS_ReactFX_IcyScarf
dependencies {
 usedDataInterfaces [
  &NiagaraDataInterfaceAudioPlayer
  &NiagaraDataInterfaceCurve
  &NiagaraDataInterfaceColorCurve
  &NiagaraDataInterfaceVectorCurve
  &NiagaraDataInterfaceSpriteRendererInfo
  &NiagaraDataInterfaceParticleRead
 ]
 usedDynamicInputs [
  &VectorFromCurve
  /Niagara/DynamicInputs/TypeConversions/MakeCustomPositionFromBool
  &FN_FrontEndVector3Modifier
 ]
 usedModules [
  &self:SystemSpawnScript.SystemScriptSource.SystemScriptGraph.NiagaraNodeAssignment_0.SetVariables_#1
  &self:SystemSpawnScript.SystemScriptSource.SystemScriptGraph.NiagaraNodeAssignment_1.SetVariables_#2
  &self:SystemSpawnScript.SystemScriptSource.SystemScriptGraph.NiagaraNodeAssignment_7.SetVariables_#3
  &self:SystemSpawnScript.SystemScriptSource.SystemScriptGraph.NiagaraNodeAssignment_4.SetVariables_#4
  &self:SystemSpawnScript.SystemScriptSource.SystemScriptGraph.NiagaraNodeAssignment_8.SetVariables_#5
  &self:SystemSpawnScript.SystemScriptSource.SystemScriptGraph.NiagaraNodeAssignment_2.SetVariables_#6
  &self:SystemSpawnScript.SystemScriptSource.SystemScriptGraph.NiagaraNodeAssignment_3.SetVariables_#7
  &self:SystemSpawnScript.SystemScriptSource.SystemScriptGraph.NiagaraNodeAssignment_9.SetVariables_#8
  &self:SystemSpawnScript.SystemScriptSource.SystemScriptGraph.NiagaraNodeAssignment_10.SetVariables_#9
  &self:SystemSpawnScript.SystemScriptSource.SystemScriptGraph.NiagaraNodeAssignment_5.SetVariables_#10
  &self:SystemSpawnScript.SystemScriptSource.SystemScriptGraph.NiagaraNodeAssignment_6.SetVariables_#11
  &SystemState
 ]
}
emitters {
 BubbleHeart {
  emitterSpawnScript [
   &InitializeAudioPlayer(Audio Player=DataInterface() {propertyValues {
      Attenuation=null
      Concurrency=null
      MaxPlaysPerTick=10
      SoundToPlay=/CosmeticCompanions/Assets/Audio_Shared/Interaction/Petting/SC_Mimosa_Interaction_React_Default
      bAllowLoopingOneShotSounds=false
      bLimitPlaysPerTick=true
      bOnlyActiveDuringGameplay=false
      bStopWhenComponentIsDestroyed=true
      ConfigurationUserParameter {parameter {name=None, typeDefHandle {registeredTypeIndex=1}, varData []}}
      ParameterNames []
     }})
   /Niagara/Modules/Audio/PlayAudio() {module=PlayAudio001}
  ]
  emitterUpdateScript [
   &EmitterState()
   &SpawnBurst_Instantaneous(Spawn Group=1, Use Spawn Probability=false, UseLoopCountLimit=false) {module=SpawnBurst_Instantaneous001}
   &SpawnBurst_Instantaneous(Use Spawn Probability=false, UseLoopCountLimit=false)
  ]
  particleSpawnScript [
   &InitializeParticle(
    Lifetime=$System.IntroLength
    Sprite Rotation Mode=::NewEnumerator0
    Sprite Size=<13 11.5>
    Sprite Size Mode=::NewEnumerator4
    UsePositionOffset=false
    Write Material Random=true
   )
   /SetVariables(Particles.VisibilityTag=&SpawnGroup()) {isSetParameters=true}
   /SetVariables(
    Particles.Snowflake Color:LinearColor=$System.Layer 3 Color
    Particles.Snowflake Size:Vector2f=&Vector2DFromFloat(Value=11)
   ) {isSetParameters=true}
  ]
  particleUpdateScript [
   &ParticleState()
   &ScaleSpriteSize(
    Scale Sprite Size Mode=::NewEnumerator0
    Uniform Scale Factor=&Add_Float(
     A=0.5
     B=&FloatFromCurve(FloatCurve=DataInterface() {propertyValues {Curve {
         defaultValue=0
         keys [
          {arriveTangent=-0.0895, interpMode=RCIM_Cubic, leaveTangent=-0.0895, tangentMode=RCTM_User, time=0.0989, value=0.125}
          {arriveTangent=1.4749, interpMode=RCIM_Cubic, leaveTangent=1.4749, tangentMode=RCTM_User, time=0.6496, value=0.875}
          {arriveTangent=0.0101, interpMode=RCIM_Cubic, leaveTangent=0.0101, tangentMode=RCTM_User, time=0.9598, value=1}
          <at 0.9931 3 RCIM_Cubic>
         ]
        }}})
    )
   ) {module=ScaleSpriteSize002}
   &ScaleColor(
    Linear Color Curve=DataInterface() {propertyValues {
      AlphaCurve {keys [<at 0 0>,<at 0.1241 1>,<at 0.95 1>,<at 1 0>]}
      BlueCurve {keys [<at 0 1>,<at 0.6346 1>,<at 0.8018 15>]}
      GreenCurve {keys [<at 0 1>,<at 0.6346 1>,<at 0.8018 15>]}
      RedCurve {keys [<at 0 1>,<at 0.6346 1>,<at 0.8018 15>]}
     }}
    Scale Mode=::NewEnumerator2
   )
   &OffsetPosition(Position Offset=&Lerp_Vector(
     A=$System.HeartIntroSpawnPosition
     B=$System.BurstPosition
     Alpha=&FloatFromCurve(FloatCurve=DataInterface() {propertyValues {Curve {keys [
          {interpMode=RCIM_Cubic, leaveTangentWeight=0.2338, tangentMode=RCTM_User, time=0, value=0}
          {arriveTangent=-0.1074, tangentMode=RCTM_User, time=0.55, value=1}
         ]}}})
    ))
   &DynamicMaterialParameters(
    Index 0 Param 1=&MergeHelper_FloatSwitch(0=1, 1=0, 2=?, 3=?, Index=&SpawnGroup(), Num=::NewEnumerator0)
    Index 0 Param 2=&MergeHelper_FloatSwitch(0=1, 1=0, 2=?, 3=?, Index=&SpawnGroup(), Num=::NewEnumerator0)
    Index 0 Param 3=&MergeHelper_FloatSwitch(0=0, 1=1, 2=?, 3=?, Index=&SpawnGroup(), Num=::NewEnumerator0)
    Module.Index0Float0Write=true
    Module.Index0Float1Write=true
    Module.Index0Float2Write=true
    Module.Index0Float3Write=true
   )
   /SetVariables(
    Particles.Glow Color=&Multiply_LinearColorByFloat(Linear Color=$System.Layer 2 Color, Float=0.4)
    Particles.Glow Size=&Multiply_Vector2D(A=$Particles.SpriteSize, B=<1.5 1.5>)
   ) {isSetParameters=true}
   /SetVariables(
    Particles.Scale Snowflake Size:Vector2f=&Vector2DFromFloat(Value=&Multiply_Float(
      A=&MakeFloatFromVector2d(Vector 2D=$Particles.Initial.Snowflake Size)
      B=&FloatFromCurve(FloatCurve=DataInterface() {propertyValues {Curve {
          defaultValue=0
          keys [
           {arriveTangent=-0.0895, interpMode=RCIM_Cubic, leaveTangent=-0.0895, tangentMode=RCTM_User, time=0.0989, value=0.125}
           {arriveTangent=1.4749, interpMode=RCIM_Cubic, leaveTangent=1.4749, tangentMode=RCTM_User, time=0.6496, value=0.875}
           {arriveTangent=0.0101, interpMode=RCIM_Cubic, leaveTangent=0.0101, tangentMode=RCTM_User, time=0.9598, value=1}
           <at 0.9931 3 RCIM_Cubic>
          ]
         }}})
     ))
    Particles.Snowflake Color:LinearColor=&ColorFromCurve(ColorCurve=DataInterface() {propertyValues {
       AlphaCurve {keys [
         <at 0.4115 0>
         <at 0.6355 0.5>
         {arriveTangent=-1.3716, interpMode=RCIM_Cubic, leaveTangent=-1.3716, time=0.8417, value=1.5}
         <at 0.9341 0.9895>
         <at 1 0 RCIM_Cubic>
        ]}
       BlueCurve {keys [<at 0 1 RCIM_Cubic>,<at 0.498 5 RCIM_Cubic>]}
       GreenCurve {keys [<at 0 1 RCIM_Cubic>,<at 0.498 5 RCIM_Cubic>]}
       RedCurve {keys [<at 0 1 RCIM_Cubic>,<at 0.498 5 RCIM_Cubic>]}
      }})
   ) {isSetParameters=true}
   &SolveForcesAndVelocity(Clamp Velocity=false, Limit Acceleration=false)
  ]
  rendererClasses [&NiagaraSpriteRendererProperties]
  renderers [
   {rendererClass=&NiagaraSpriteRendererProperties}
   {rendererClass=&NiagaraSpriteRendererProperties, rendererIndex=1}
   {rendererClass=&NiagaraSpriteRendererProperties, rendererIndex=2}
  ]
 }
 Pop {
  emitterUpdateScript [
   &EmitterState()
   &SpawnBurst_Instantaneous(Spawn Group=5, Spawn Time=1, Use Spawn Probability=false, UseLoopCountLimit=false) {module=SpawnBurst_Instantaneous005}
   &SpawnBurst_Instantaneous(Spawn Group=4, Spawn Time=1, Use Spawn Probability=false, UseLoopCountLimit=false) {module=SpawnBurst_Instantaneous004}
   &SpawnBurst_Instantaneous(Spawn Group=3, Spawn Time=1, Use Spawn Probability=false, UseLoopCountLimit=false) {module=SpawnBurst_Instantaneous003}
   &SpawnBurst_Instantaneous(Spawn Count=8, Spawn Group=1, Spawn Time=0.95, Use Spawn Probability=false, UseLoopCountLimit=false)
   &SpawnBurst_Instantaneous(Spawn Count=8, Spawn Time=0.95, Use Spawn Probability=false, UseLoopCountLimit=false) {module=SpawnBurst_Instantaneous001}
   &SpawnBurst_Instantaneous(Spawn Count=5, Spawn Group=2, Spawn Time=0.95, Use Spawn Probability=false, UseLoopCountLimit=false) {module=SpawnBurst_Instantaneous002}
  ]
  particleSpawnScript [
   /SetVariables(Particles.VisibilityTag=&SpawnGroup()) {isSetParameters=true}
   &InitializeParticle(
    Color=&FNTimeOfDayColor(
     Night Color=<rgba 0.4271 0.4271 0.4271 1>
     Morning Color=<rgba 0.4635 0.7625 1 1>
     Day Color=<rgba 0.4635 0.7625 1 1>
     Evening Color=<rgba 0.4635 0.7625 1 1>
    )
    Lifetime=&MergeHelper_FloatSwitch(
     0=&RandomRangeFloat(Minimum=0.25, Maximum=0.8)
     1=&RandomRangeFloat(Minimum=0.18, Maximum=0.3)
     2=&RandomRangeFloat(Minimum=0.25, Maximum=0.8)
     3=0.15
     4=0.2
     5=0.15
     Index=&SpawnGroup()
     Num=::NewEnumerator4
    )
    Mesh Scale Mode=::NewEnumerator1
    Mesh Uniform Scale Max=0.3
    Mesh Uniform Scale Min=0.15
    Position Offset=$System.BurstPosition
    Sprite Rotation Mode=::NewEnumerator2
    Sprite Size Mode=::NewEnumerator1
    Uniform Sprite Size Max=6
    Uniform Sprite Size Min=3
    UsePositionOffset=true
    Write Material Random=true
   )
   &AddVelocity(
    Module.Use Falloff Distance=false
    Origin Offset=$System.BurstPosition
    Override Seed=false
    Velocity Mode=::NewEnumerator1
    Velocity Speed=&MergeHelper_FloatSwitch(
     0=&RandomRangeFloat(Minimum=70, Maximum=110)
     1=&RandomRangeFloat(Minimum=80, Maximum=200)
     2=&RandomRangeFloat(Minimum=70, Maximum=110)
     3=0
     4=0
     5=0
     Index=&SpawnGroup()
     Num=::NewEnumerator4
    )
   )
   /SetVariables(Particles.Glow Size=&Vector2DFromFloat(Value=&RandomRangeFloat(Minimum=15, Maximum=20))) {isSetParameters=true}
   /SetVariables(
    Particles.AirGlow Dynam:Vector4f=&MakeVector4(X=2, Y=0, Z=0, W=0)
    Particles.Airglow Color:LinearColor=<rgba 0.099 0.7793 1 0.5>
    Particles.Airglow Size:Vector2f=&Vector2DFromFloat(Value=65)
   ) {isSetParameters=true}
   /SetVariables(
    Particles.Rays Color:LinearColor=<rgba 0.3694 0.8455 1 1>
    Particles.Rays Dynam:Vector4f=&MakeVector4(
     X=&RandomRangeFloat(Minimum=0, Maximum=10)
     Y=&RandomRangeFloat(Minimum=2.5, Maximum=4.5)
     Z=&RandomRangeFloat(Minimum=0.8, Maximum=1.6)
     W=0
    )
    Particles.Rays Size:Vector3f=&VectorFromFloat(Value=0.6)
   ) {isSetParameters=true}
   /SetVariables(
    Particles.Corona Color:LinearColor=<rgba 1 0.7672 0.4943 1>
    Particles.Corona Size:Vector2f=&Vector2DFromFloat(Value=55)
   ) {isSetParameters=true}
  ]
  particleUpdateScript [
   &ParticleState()
   &UpdateMeshOrientation(
    Rotation Rate=&MergeHelper_FloatSwitch(0=1, 1=0, 2=1, 3=0, 4=0, 5=0, Index=&SpawnGroup(), Num=::NewEnumerator4)
    Rotation Vector=&MergeHelper_VectorSwitch(0=<1 1 1>, 1=<0 0 0>, 2=<1 1 1>, 3=<0 0 0>, 4=<0 0 0>, 5=<0 0 0>, Index=&SpawnGroup(), Num=::NewEnumerator4)
   )
   &SpriteRotationRate(Rotation Rate=&RandomRangeFloat(Minimum=-50, Maximum=50))
   &ScaleMeshSize(Scale Factor=&VectorFromFloat(Value=&FloatFromCurve(FloatCurve=DataInterface() {propertyValues {Curve {keys [<at 0 1>,<at 0.7964 1>,<at 1 0>]}}})))
   &ScaleSpriteSize(
    Scale Sprite Size Mode=::NewEnumerator0
    Uniform Scale Factor=&FloatFromCurve(FloatCurve=DataInterface() {propertyValues {Curve {keys [
         <at 0 0 RCIM_Cubic>
         {arriveTangentWeight=0.0433, interpMode=RCIM_Cubic, leaveTangentWeight=0.3986, tangentMode=RCTM_User, time=0.0852, value=1}
         {arriveTangentWeight=0.3878, interpMode=RCIM_Cubic, leaveTangentWeight=0.064, tangentMode=RCTM_User, time=0.8704, value=1}
         <at 1 0 RCIM_Cubic>
        ]}}})
   ) {module=ScaleSpriteSize002}
   &ScaleColor(
    Linear Color Curve=DataInterface() {propertyValues {
      AlphaCurve {keys [<at 0 0>,<at 0.1364 0.5>,<at 0.4198 0.5>,<at 1 0>]}
      BlueCurve {keys [<at 0 0.901>]}
      GreenCurve {keys [<at 0 0.901>]}
      RedCurve {keys [<at 0 0.901>]}
     }}
    Scale Mode=::NewEnumerator2
   )
   &Drag(
    Drag=&MergeHelper_FloatSwitch(
     0=&FloatFromCurve(
      FloatCurve=DataInterface() {propertyValues {Curve {keys [
          {interpMode=RCIM_Cubic, leaveTangentWeight=7.3294, tangentMode=RCTM_User, time=0.2181, value=0}
          {
           arriveTangent=119.2046
           arriveTangentWeight=7.3294
           interpMode=RCIM_Cubic
           leaveTangent=119.2046
           tangentMode=RCTM_User
           time=0.4999
           value=10
          }
         ]}}}
      Scale Curve=&RandomRangeFloat(Minimum=0.5, Maximum=1)
     )
     1=0
     2=&FloatFromCurve(
      FloatCurve=DataInterface() {propertyValues {Curve {keys [
          {interpMode=RCIM_Cubic, leaveTangentWeight=7.3294, tangentMode=RCTM_User, time=0.2181, value=0}
          {
           arriveTangent=119.2046
           arriveTangentWeight=7.3294
           interpMode=RCIM_Cubic
           leaveTangent=119.2046
           tangentMode=RCTM_User
           time=0.4999
           value=10
          }
         ]}}}
      Scale Curve=&RandomRangeFloat(Minimum=0.5, Maximum=1)
     )
     3=0
     4=0
     5=0
     Index=&SpawnGroup()
     Num=::NewEnumerator4
    )
    Use Linear Drag=true
    Use Rotational Drag=true
   )
   &SolveForcesAndVelocity(Clamp Velocity=false, Limit Acceleration=false)
   &AlignSpriteToMeshOrientation(
    Mesh Orientation Relative Sprite Facing Vector:Vector3f=&MergeHelper_VectorSwitch(
     0=&RandomRangeVector(Minimum=<-1 -1 -1>, Maximum=<1 1 1>)
     1=<0 0 0>
     2=<0 0 0>
     3=<0 0 0>
     4=<0 0 0>
     5=<0 0 0>
     Index=&SpawnGroup()
     Num=::NewEnumerator4
    )
    SetAlignment=true
    SetFacing=true
   )
   /SetVariables(Particles.Glow Color=&Multiply_LinearColorByFloat(
     Linear Color=&MultiplyLinearColor(
      Linear Color=$Particles.Color
      Scale Factor=&MakeVector4_FromLinearColor(LinearColor=<rgba 0.5412 1 0.9368 1>)
     )
     Float=0.33
    )) {isSetParameters=true}
   &SubUVAnimation(
    Evaluation Type=::NewEnumerator0
    Output Alternate Renderer Binding=true
    "Sprite Renderer "=DataInterface() {propertyValues {SpriteRenderer=&self:Pop_0.NiagaraSpriteRendererProperties_0}}
    SubUV Animation Mode:ENiagara_SubUVLookupModeV2=::NewEnumerator2
    UseEndFrame=false
    UseStartFrame=false
   )
   /SetVariables(Particles.Scale Airglow Color:LinearColor=&Multiply_LinearColorByFloat(
     Linear Color=$Particles.Initial.Airglow Color
     Float=&FloatFromCurve(FloatCurve=DataInterface() {propertyValues {Curve {keys [<at 0 0 RCIM_Cubic>,<at 0.25 1 RCIM_Cubic>,<at 1 0 RCIM_Cubic>]}}})
    )) {isSetParameters=true}
   /SetVariables(
    Particles.Scale Rays Color:LinearColor=&Multiply_LinearColorByFloat(
     Linear Color=$Particles.Initial.Rays Color
     Float=&FloatFromCurve(FloatCurve=DataInterface() {propertyValues {Curve {keys [
          <at 0 0 RCIM_Cubic>
          {arriveTangent=2, interpMode=RCIM_Cubic, leaveTangent=2, time=0.2009, value=0.6}
          <at 0.5 1 RCIM_Cubic>
          {arriveTangent=-2, interpMode=RCIM_Cubic, leaveTangent=-2, time=0.8015, value=0.6}
          <at 1 0 RCIM_Cubic>
         ]}}})
    )
    Particles.Scale Rays Size:Vector3f=&Multiply_VectorByFloat(
     Vector=$Particles.Initial.Rays Size
     Float=&FloatFromCurve(FloatCurve=DataInterface() {propertyValues {Curve {keys [
          <at 0 0 RCIM_Cubic>
          {arriveTangentWeight=0.0571, interpMode=RCIM_Cubic, leaveTangentWeight=0.2418, tangentMode=RCTM_User, time=0.0938, value=1}
          {arriveTangent=-0.4414, interpMode=RCIM_Cubic, leaveTangent=-0.4414, time=0.4913, value=0.75}
          <at 1 0.6 RCIM_Cubic>
         ]}}})
    )
   ) {isSetParameters=true}
   /SetVariables(
    Particles.Scale Corona Color:LinearColor=&Multiply_LinearColorByFloat(
     Linear Color=$Particles.Initial.Corona Color
     Float=&FloatFromCurve(FloatCurve=DataInterface() {propertyValues {Curve {keys [<at 0 0.45>,<at 0.2 1>,<at 0.5 0.25>,<at 1 0>]}}})
    )
    Particles.Scale Corona Size:Vector2f=&Multiply_Vector2D(
     A=$Particles.Initial.Corona Size
     B=&Vector2DFromFloat(Value=&FloatFromCurve(FloatCurve=DataInterface() {propertyValues {Curve {keys [<at 0 0 RCIM_Cubic>, {arriveTangent=1, interpMode=RCIM_Cubic, leaveTangent=1, time=0.108, value=0.6}, <at 1 1 RCIM_Cubic>]}}}))
    )
   ) {isSetParameters=true}
  ]
  rendererClasses [&NiagaraSpriteRendererProperties, &NiagaraMeshRendererProperties]
  renderers [
   {rendererClass=&NiagaraSpriteRendererProperties}
   {rendererClass=&NiagaraSpriteRendererProperties, rendererIndex=1}
   {rendererClass=&NiagaraMeshRendererProperties, rendererIndex=2}
   {rendererClass=&NiagaraSpriteRendererProperties, rendererIndex=3}
   {rendererClass=&NiagaraMeshRendererProperties, rendererIndex=4}
   {rendererClass=&NiagaraSpriteRendererProperties, rendererIndex=5}
  ]
 }
 Snow_Ribbon {
  emitterUpdateScript [
   &EmitterState()
   &SpawnParticlesFromOtherEmitter(
    Attribute Reader=DataInterface() {propertyValues {EmitterBinding {bindingMode=Other, emitterName=BubbleHeart}}}
    Spawn Rate=100
    Use Maximum Age=true
    Use Minimum Age=false
   )
  ]
  particleSpawnScript [
   &InitializeParticle(
    Color=<rgba 0.3185 0.7453 1 1>
    Lifetime=0.26
    Ribbon Width=30
    Ribbon Width Mode=::NewEnumerator1
    Sprite Rotation Mode=::NewEnumerator2
    Sprite Size Mode=::NewEnumerator1
    Uniform Sprite Size Max=12
    UsePositionOffset=false
    Write Material Random=true
   )
   &SampleParticlesFromOtherEmitter()
   &AddVelocity(Velocity=<0 0 -300>)
  ]
  particleUpdateScript [
   &ParticleState()
   &ScaleColor(
    Scale Alpha=&FloatFromCurve(
     FloatCurve=DataInterface() {propertyValues {Curve {
        defaultValue=1
        keys [
         {interpMode=RCIM_Cubic, tangentMode=RCTM_User, time=0.125, value=0}
         <at 0.408 1 RCIM_Cubic>
         {arriveTangent=-2.8205, interpMode=RCIM_Cubic, leaveTangent=-2.8205, tangentMode=RCTM_User, time=0.7, value=0}
        ]
       }}}
     Scale Curve=0.3
    )
    ScaleA=true
    ScaleRGB=false
   )
   &ScaleColor(
    Scale Alpha=&FloatFromCurve(
     FloatCurve=DataInterface() {propertyValues {Curve {
        defaultValue=1
        keys [
         {interpMode=RCIM_Cubic, tangentMode=RCTM_User, time=0, value=1}
         {arriveTangentWeight=0.8215, interpMode=RCIM_Cubic, leaveTangentWeight=0.0762, tangentMode=RCTM_User, time=0.9284, value=1}
         <at 0.9845 0 RCIM_Cubic>
        ]
       }}}
     CurveIndex=$System.Age
    )
    ScaleA=true
    ScaleRGB=false
   ) {module=ScaleColor001}
   &SolveForcesAndVelocity(Clamp Velocity=false, Limit Acceleration=false)
   &DynamicMaterialParameters(
    Index 0 Param 1=&FloatFromCurve(FloatCurve=DataInterface() {propertyValues {Curve {keys [<at 0 1>,<at 1 0>]}}})
    Index 0 Param 3=-1.9163
    Index 0 Param 4=0.2
    Module.Index0Float0Write=true
    Module.Index0Float1Write=true
    Module.Index0Float2Write=true
    Module.Index0Float3Write=true
    Module.Index1Float0Write=true
    Module.Index1Float1Write=true
    Module.Index1Float2Write=true
    Module.Index1Float3Write=true
    Write Parameter Index 1=::NewEnumerator1
   )
   &ScaleRibbonWidth(Ribbon Width Scale=&FloatFromCurve(FloatCurve=DataInterface() {propertyValues {Curve {keys [<at 0 1>,<at 0.3891 1>,<at 0.8358 0>]}}}))
   &ScaleRibbonWidth(Ribbon Width Scale=&FloatFromCurve(FloatCurve=DataInterface() {propertyValues {Curve {keys [<at 0 0.4>,<at 0.5 1>]}}}, CurveIndex=$System.Age)) {module=ScaleRibbonWidth001}
  ]
  rendererClasses [&NiagaraRibbonRendererProperties]
  renderers [{rendererClass=&NiagaraRibbonRendererProperties}]
 }
 SpawnBurst {
  emitterUpdateScript [
   &EmitterState()
   &SpawnBurst_Instantaneous(Spawn Count=8, Spawn Group=1, Use Spawn Probability=false, UseLoopCountLimit=false) {module=SpawnBurst_Instantaneous001}
   &SpawnBurst_Instantaneous(Spawn Count=6, Use Spawn Probability=false, UseLoopCountLimit=false)
  ]
  particleSpawnScript [
   &InitializeParticle(
    Color=&FNTimeOfDayColor(
     Night Color=<rgba 0.4271 0.4271 0.4271 1>
     Morning Color=<rgba 0.4635 0.7625 1 1>
     Day Color=<rgba 0.4635 0.7625 1 0.6>
     Evening Color=<rgba 0.4635 0.7625 1 1>
    )
    Lifetime Max=1
    Lifetime Min=0.75
    Lifetime Mode=::NewEnumerator1
    Position Offset=&SubtractVectorFromPosition(Position=Linked() {linkedVariable {name=System.HeartIntroSpawnPosition, type=Vector3f}}, Vector=<0 0 50>)
    Sprite Rotation Mode=::NewEnumerator2
    Sprite Size Mode=::NewEnumerator1
    Uniform Sprite Size Max=5
    Uniform Sprite Size Min=3
    UsePositionOffset=true
    Write Material Random=true
   )
   &AddVelocity(
    Cone Axis=<0 0 1>
    Module.UseConeFalloff=false
    Module.UseVelDistribution=false
    Override Seed=false
    Velocity Mode=::NewEnumerator2
    Velocity Speed=&RandomRangeFloat(Minimum=300, Maximum=470)
   )
   &SubUVAnimation(
    Evaluation Type=::NewEnumerator0
    "Sprite Renderer "=DataInterface() {propertyValues {SpriteRenderer=&self:SpawnBurst_0.NiagaraSpriteRendererProperties_0}}
    SubUV Animation Mode:ENiagara_SubUVLookupModeV2=::NewEnumerator2
    UseEndFrame=false
    UseStartFrame=false
   )
   /SetVariables(Particles.VisibilityTag=&SpawnGroup()) {isSetParameters=true}
   /SetVariables(Particles.Glow Size=&Vector2DFromFloat(Value=&RandomRangeFloat(Minimum=4, Maximum=8))) {isSetParameters=true}
  ]
  particleUpdateScript [
   &ParticleState()
   &ScaleSpriteSize(Uniform Curve Sprite Scale=DataInterface() {propertyValues {Curve {keys [
        <at 0 0 RCIM_Cubic>
        {arriveTangentWeight=0.1067, interpMode=RCIM_Cubic, leaveTangentWeight=0.3202, tangentMode=RCTM_User, time=0.2, value=1}
        {arriveTangentWeight=0.3202, interpMode=RCIM_Cubic, leaveTangentWeight=0.1067, tangentMode=RCTM_User, time=0.8, value=1}
        <at 1 0 RCIM_Cubic>
       ]}}})
   &ScaleSpriteSize(Uniform Curve Sprite Scale=DataInterface() {propertyValues {Curve {keys [
        <at 0 0 RCIM_Cubic>
        {arriveTangentWeight=0.1067, interpMode=RCIM_Cubic, leaveTangentWeight=0.3202, tangentMode=RCTM_User, time=0.2, value=1}
        {arriveTangentWeight=0.3202, interpMode=RCIM_Cubic, leaveTangentWeight=0.1067, tangentMode=RCTM_User, time=0.8, value=1}
        <at 1 0 RCIM_Cubic>
       ]}}}) {module=ScaleSpriteSize001}
   &VortexForce(
    Origin Pull Amount=2542.4172
    UseOriginOffset=true
    Vortex Force Amount=&RandomRangeFloat(Minimum=2000, Maximum=3000)
    Vortex Origin Offset:Vector3f=&SubtractVectorFromPosition(Position=Linked() {linkedVariable {name=System.HeartIntroSpawnPosition, type=Vector3f}}, Vector=<0 0 50>)
   )
   &Drag(
    Drag=&RandomRangeFloat(Minimum=2.8, Maximum=3)
    Rotational Drag=5
    Use Linear Drag=true
    Use Rotational Drag=true
   )
   &SolveForcesAndVelocity(Clamp Velocity=false, Limit Acceleration=false)
   &ScaleColor(
    Linear Color Curve=DataInterface() {propertyValues {
      AlphaCurve {keys [<at 0 0>,<at 0.1218 0.7>,<at 0.803 0.7>,<at 1 0>]}
      BlueCurve {keys [<at 0 1>,<at 1 1>]}
      GreenCurve {keys [<at 0 1>,<at 1 1>]}
      RedCurve {keys [<at 0 1>,<at 1 1>]}
     }}
    Scale Mode=2
   )
  ]
  rendererClasses [&NiagaraSpriteRendererProperties]
  renderers [
   {rendererClass=&NiagaraSpriteRendererProperties}
   {rendererClass=&NiagaraSpriteRendererProperties, rendererIndex=1}
  ]
 }
}
summary {
 systemName=NS_ReactFX_IcyScarf
 emitters [
  {emitterName=BubbleHeart, rendererClasses [&NiagaraSpriteRendererProperties]}
  {emitterName=Pop, rendererClasses [&NiagaraSpriteRendererProperties, &NiagaraMeshRendererProperties]}
  {emitterName=SpawnBurst, rendererClasses [&NiagaraSpriteRendererProperties]}
  {emitterName=Snow_Ribbon, rendererClasses [&NiagaraRibbonRendererProperties]}
 ]
 userVariables [
  User.bIsFrontEnd {type=NiagaraBool}
  User.bIsFrontEndPreview {type=NiagaraBool}
  User.FrontEndScale {type=NiagaraFloat, defaultValue=0.6, description=""}
 ]
}
