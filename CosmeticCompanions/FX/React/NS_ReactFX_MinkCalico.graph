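# Niagara reaction effect for the MinkCalico cosmetic companion, playing a petting reaction sound with a warm gold flourish.
# An Intro sprite travels from the heart spawn position to the burst position, feeding a ribbon Trail and pixie dust through particle reads.
# Rotating spiral meshes rise from the start, and a delayed packed burst adds sprites and meshes at the end point.
# Packed emitters share renderers via spawn-group tags; brightness follows time of day, and some spawn counts vary by platform.
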NiagaraSystem /CosmeticCompanions/FX/React/NS_ReactFX_MinkCalico
dependencies {
 usedDataInterfaces [
  &NiagaraDataInterfaceAudioPlayer
  &NiagaraDataInterfaceSpriteRendererInfo
  &NiagaraDataInterfaceCurve
  &NiagaraDataInterfaceVector2DCurve
  &NiagaraDataInterfaceParticleRead
  &NiagaraDataInterfaceColorCurve
  &NiagaraDataInterfacePlatformSet
 ]
 usedDynamicInputs [
  &TransformVector
  &ReturnNormalizedExecIndex
  &FN_FrontEndVector3Modifier
 ]
 usedModules [
  &self:SystemSpawnScript.SystemScriptSource.SystemScriptGraph.NiagaraNodeAssignment_2.SetVariables_#1
  &self:SystemSpawnScript.SystemScriptSource.SystemScriptGraph.NiagaraNodeAssignment_0.SetVariables_#2
  &self:SystemSpawnScript.SystemScriptSource.SystemScriptGraph.NiagaraNodeAssignment_4.SetVariables_#3
  &self:SystemSpawnScript.SystemScriptSource.SystemScriptGraph.NiagaraNodeAssignment_3.SetVariables_#4
  &self:SystemSpawnScript.SystemScriptSource.SystemScriptGraph.NiagaraNodeAssignment_1.SetVariables_#5
  &SystemState
 ]
}
emitters {
 Burst_Packed {
  emitterUpdateScript [
   &EmitterState(Life Cycle Mode=::NewEnumerator1, Loop Behavior=::NewEnumerator1, Loop Delay=0.69, Loop Duration=0.4, UseLoopDelay=true)
   &SpawnBurst_Instantaneous(Use Spawn Probability=false, UseLoopCountLimit=false)
   &SpawnBurst_Instantaneous(Spawn Group=1, Use Spawn Probability=false, UseLoopCountLimit=false) {module=SpawnBurst_Instantaneous001}
   &SpawnBurst_Instantaneous(
    Spawn Count=&MakeIntFromFloat(Float=&FN_PickFloatDependOnPlatform(
      TargetPlatformActive=&MakeFloatFromInt(Int=&RandomRangeInt(Minimum=6, Maximum=8))
      TargetPlatformNoActive=&MakeFloatFromInt(Int=&RandomRangeInt(Minimum=3, Maximum=4))
     ))
    Spawn Group=2
    Use Spawn Probability=false
    UseLoopCountLimit=false
   ) {module=SpawnBurst_Instantaneous002}
   &SpawnBurst_Instantaneous(
    Spawn Count=&MakeIntFromFloat(Float=&FN_PickFloatDependOnPlatform(
      TargetPlatformActive=&MakeFloatFromInt(Int=&RandomRangeInt(Minimum=6, Maximum=8))
      TargetPlatformNoActive=&MakeFloatFromInt(Int=&RandomRangeInt(Minimum=3, Maximum=4))
     ))
    Spawn Group=3
    Use Spawn Probability=false
    UseLoopCountLimit=false
   ) {module=SpawnBurst_Instantaneous003}
  ]
  particleSpawnScript [
   /SetVariables(Particles.VisibilityTag=$Emitter.SpawnGroup) {isSetParameters=true}
   &InitializeParticle(
    Color Mode=::NewEnumerator0
    Lifetime=&MergeHelper_FloatSwitch(
     0=0.4
     1=0.5
     2=&RandomRangeFloat(Minimum=0.45, Maximum=0.7)
     3=&RandomRangeFloat(Minimum=0.45, Maximum=0.7)
     Index=$Particles.VisibilityTag
     Num=::NewEnumerator2
    )
    Position Offset=$System.BurstPosition
    Position Offset Coordinate Space=::World
    Sprite Rotation Mode=::NewEnumerator2
    Sprite Size Mode=::NewEnumerator3
    Sprite UV Mode=::NewEnumerator3
    Uniform Sprite Size=45
    UsePositionOffset=true
    Write Material Random=true
   )
   &CameraOffset(Camera Offset Amount=15)
   &InitialMeshOrientation(
    Mesh Orientation Mode=::NewEnumerator7
    Rotation=&RandomRangeVector(Minimum=<0 0 0>, Maximum=<1 0 0>)
    Use Rotation Vector=true
   )
   &DynamicMaterialParameters(
    Dynamic Parameter Output:ENiagara_AttributeSamplingApplyOutput=::NewEnumerator1
    Index 0 Param 1=&RandomRangeFloat(Minimum=-0.05, Maximum=0.05)
    Index 0 Param 2=&RandomRangeFloat(Minimum=0, Maximum=32)
    Module.Index0Float0Write=true
    Module.Index0Float1Write=true
    Module.Index0Float2Write=true
    Module.Index0Float3Write=true
   ) {module=DynamicMaterialParameters001}
   &ShapeLocation(
    MaskSpawnGroup=false
    Override Seed=false
    Sphere Radius=&MergeHelper_FloatSwitch(0=0, 1=0, 2=1, 3=1, Index=$Particles.VisibilityTag, Num=::NewEnumerator2)
    Sphere Surface Distribution=1
    UseOwnerScale=true
   )
   &AddVelocity(
    Module.Use Falloff Distance=false
    Origin Offset=$System.BurstPosition
    Origin Offset Coordinate Space=::Simulation
    Override Seed=false
    Velocity Mode=::NewEnumerator1
    Velocity Speed=&MergeHelper_FloatSwitch(
     0=0
     1=0
     2=&RandomRangeFloat(Minimum=45, Maximum=75)
     3=&RandomRangeFloat(Minimum=45, Maximum=75)
     Index=$Particles.VisibilityTag
     Num=::NewEnumerator2
    )
   )
   &SubUVAnimation(
    End Frame Range Override=1
    Evaluation Type=::NewEnumerator0
    Output Alternate Renderer Binding=true
    "Sprite Renderer "=DataInterface() {propertyValues {SpriteRenderer=&self:Burst_Packed_0.NiagaraSpriteRendererProperties_3}}
    SubUV Animation Mode:ENiagara_SubUVLookupModeV2=::NewEnumerator2
    UseEndFrame=true
    UseStartFrame=true
   )
  ]
  particleUpdateScript [
   &ParticleState()
   &Drag(
    Drag=&MergeHelper_FloatSwitch(
     0=0
     1=0
     2=&RandomRangeFloat(Minimum=3, Maximum=5)
     3=&RandomRangeFloat(Minimum=3, Maximum=5)
     Index=$Particles.VisibilityTag
     Num=::NewEnumerator2
    )
    Use Linear Drag=true
    Use Rotational Drag=true
   )
   &DynamicMaterialParameters(
    Index 0 Param 1=1
    Index 0 Param 2=1
    Index 0 Param 3=1
    Index 0 Param 4=1
    Module.Index0Float0Write=true
    Module.Index0Float1Write=true
    Module.Index0Float2Write=true
    Module.Index0Float3Write=true
   )
   &ScaleSpriteSize(Uniform Curve Sprite Scale=DataInterface() {propertyValues {Curve {keys [
        {arriveTangent=2.0035, interpMode=RCIM_Cubic, leaveTangent=2.0035, tangentMode=RCTM_User, time=0, value=1}
        {arriveTangent=-4.2244, interpMode=RCIM_Cubic, leaveTangent=-4.2244, tangentMode=RCTM_User, time=0.1, value=1.35}
        {tangentMode=RCTM_User, time=1, value=0}
       ]}}})
   &ScaleColor(
    Color Scale:LinearColor=&ColorFromCurve(ColorCurve=DataInterface() {propertyValues {
       AlphaCurve {keys [<at 0 1 RCIM_Cubic>,<at 1 0 RCIM_Cubic>]}
       BlueCurve {keys [<at 0 0 RCIM_Cubic>,<at 1 0.6 RCIM_Cubic>]}
       GreenCurve {keys [<at 0 2.15 RCIM_Cubic>,<at 1 0.9333 RCIM_Cubic>]}
       RedCurve {keys [<at 0 3 RCIM_Cubic>,<at 1 1 RCIM_Cubic>]}
      }})
    Scale Mode=3
   )
   &FN_Mesh_Attributes(
    Color=&MakeLinearColorFromVectorAndFloat(
     "Vector (RGB)"=&Multiply_VectorByFloat(
      Vector=&MakeVectorFromLinearColorRGB(Color=&ColorFromCurve(ColorCurve=DataInterface() {propertyValues {
          AlphaCurve {keys [<at 0 1 RCIM_Cubic>]}
          BlueCurve {keys [<at 0 0 RCIM_Cubic>,<at 1 0.6 RCIM_Cubic>]}
          GreenCurve {keys [<at 0 2.15 RCIM_Cubic>,<at 1 0.9333 RCIM_Cubic>]}
          RedCurve {keys [<at 0 3 RCIM_Cubic>,<at 1 1 RCIM_Cubic>]}
         }}))
      Float=&FloatFromCurve(FloatCurve=DataInterface() {propertyValues {Curve {keys [<at 0 5>,<at 0.15 1.5>,<at 1 1>]}}})
     )
     "Float (Alpha)"=&FloatFromCurve(
      FloatCurve=DataInterface() {propertyValues {Curve {keys [<at 0 1>,<at 1 0>]}}}
      Scale Curve=&Multiply_Float(A=0.1, B=&FNTimeOfDayScalar(Night Scale=0.75, Morning Scale=0.8, Evening Scale=0.9))
     )
    )
    Create Camera Offset Attribute=true
    Create Color Attribute=true
    Create Dynamic Parameter Attribute=true
    Create Scale Attribute=true
    Index 0 Param 1:NiagaraFloat=&FloatFromCurve(FloatCurve=DataInterface() {propertyValues {Curve {keys [{arriveTangent=-2.1015, interpMode=RCIM_Cubic, leaveTangent=-2.1015, tangentMode=RCTM_User, time=0, value=0.5}, <at 1 -0.15>]}}})
    Index 0 Param 2:NiagaraFloat=&FloatFromCurve(FloatCurve=DataInterface() {propertyValues {Curve {keys [{arriveTangent=-0.3464, interpMode=RCIM_Cubic, leaveTangent=-0.3464, tangentMode=RCTM_User, time=0, value=1}, <at 1 0.9>]}}})
    Index 0 Param 3:NiagaraFloat=&FloatFromCurve(FloatCurve=DataInterface() {propertyValues {Curve {keys [<at 0 0>,<at 0.15 1>,<at 1 1.2>]}}}, Scale Curve=0.5)
    Module Output=::NewEnumerator1
    Module.Index0Float0Write=true
    Module.Index0Float1Write=true
    Module.Index0Float2Write=true
    Module.Index0Float3Write=true
    Scale=&VectorFromFloat(Value=&FloatFromCurve(FloatCurve=DataInterface() {propertyValues {Curve {keys [<at 0 0>,<at 0.1 1.4>,<at 1 2>]}}}, Scale Curve=0.2))
   )
   &FN_Sprite_Attributes(
    Camera Offset=-5
    Color=&MultiplyLinearColor(
     Linear Color=&HueShiftLinearColor(
      Hue=&RandomRangeFloat(Minimum=-0.05, Maximum=0.05)
      Saturation=&RandomRangeFloat(Minimum=0.8, Maximum=1)
      Value=&RandomRangeFloat(Minimum=1, Maximum=1.3)
      Linear Color=&RandomRangeLinearColor(Minimum=<rgba 50 46.6667 30 0.8>, Maximum=<rgba 50 46.6667 30 1>)
     )
     Scale Factor=&MakeVector4FromVectorAndFloat(
      "Vector (RGB)"=<1 1 1>
      "Float (Alpha)"=&FloatFromCurve(
       FloatCurve=DataInterface() {propertyValues {Curve {keys [
           {interpMode=RCIM_Cubic, leaveTangentWeight=0.0333, tangentMode=RCTM_User, time=0, value=0}
           {arriveTangentWeight=0.0333, interpMode=RCIM_Cubic, leaveTangentWeight=0.3, tangentMode=RCTM_User, time=0.1, value=1}
           {arriveTangentWeight=0.3, interpMode=RCIM_Cubic, tangentMode=RCTM_User, time=1, value=0}
          ]}}}
       Scale Curve=10
      )
     )
    )
    Create Camera Offset Attribute=true
    Create Color Attribute=true
    Create SpriteSize Attribute=true
    Module Output:ENiagara_AttributeSamplingApplyOutput=::NewEnumerator1
    SpriteSize=&Vector2DFromFloat(Value=&Multiply_Float(
      A=&RandomRangeFloat(Minimum=18.5, Maximum=25)
      B=&FloatFromCurve(FloatCurve=DataInterface() {propertyValues {Curve {keys [
           {arriveTangent=2.2332, interpMode=RCIM_Cubic, leaveTangent=2.2332, tangentMode=RCTM_User, time=0, value=0.6}
           {tangentMode=RCTM_User, time=1, value=1.25}
          ]}}})
     ))
   )
   &FN_Sprite_Attributes(
    Color=&MakeLinearColorFromVectorAndFloat(
     "Vector (RGB)"=<1 1 1>
     "Float (Alpha)"=&FloatFromCurve(
      FloatCurve=DataInterface() {propertyValues {Curve {keys [
          {interpMode=RCIM_Cubic, leaveTangentWeight=0.0333, tangentMode=RCTM_User, time=0, value=0}
          {arriveTangentWeight=0.0333, interpMode=RCIM_Cubic, leaveTangentWeight=0.3, tangentMode=RCTM_User, time=0.1, value=1}
          {arriveTangentWeight=0.3, interpMode=RCIM_Cubic, tangentMode=RCTM_User, time=1, value=0}
         ]}}}
      Scale Curve=$System.ToDScalar
     )
    )
    Create Color Attribute=true
    Create SpriteSize Attribute=true
    Module Output:ENiagara_AttributeSamplingApplyOutput=::NewEnumerator1
    SpriteSize=&Vector2DFromFloat(Value=&Multiply_Float(
      A=&RandomRangeFloat(Minimum=3, Maximum=5)
      B=&FloatFromCurve(FloatCurve=DataInterface() {propertyValues {Curve {keys [
           {arriveTangent=18.3146, interpMode=RCIM_Cubic, leaveTangent=18.3146, tangentMode=RCTM_User, time=0, value=0.6}
           {arriveTangent=-3.1654, interpMode=RCIM_Cubic, leaveTangent=-3.1654, tangentMode=RCTM_User, time=0.3, value=1.45}
           {tangentMode=RCTM_User, time=1, value=0}
          ]}}})
     ))
   ) {module=FN_Sprite_Attributes002}
   &FN_Sprite_Attributes(
    Color=&MultiplyLinearColor(
     Linear Color=<rgba 1 0.4945 0 2>
     Scale Factor=&MakeVector4FromVectorAndFloat("Vector (RGB)"=<1 1 1>, "Float (Alpha)"=$System.ToDScalar)
    )
    Create Color Attribute=true
    Module Output:ENiagara_AttributeSamplingApplyOutput=::NewEnumerator1
   ) {module=FN_Sprite_Attributes001}
   &CurlNoiseForce(
    Bias Noise=false
    Noise Frequency=15
    Noise Strength=&MergeHelper_FloatSwitch(
     0=0
     1=0
     2=0
     3=&FloatFromCurve(FloatCurve=DataInterface() {propertyValues {Curve {keys [<at 0 100>,<at 1 0>]}}})
     Index=$Particles.VisibilityTag
     Num=::NewEnumerator2
    )
    Pan Noise=false
   )
   &SolveForcesAndVelocity(Clamp Velocity=false, Limit Acceleration=false)
  ]
  rendererClasses [&NiagaraSpriteRendererProperties, &NiagaraMeshRendererProperties]
  renderers [
   {rendererClass=&NiagaraSpriteRendererProperties}
   {rendererClass=&NiagaraMeshRendererProperties, rendererIndex=1}
   {rendererClass=&NiagaraSpriteRendererProperties, rendererIndex=2}
   {rendererClass=&NiagaraSpriteRendererProperties, rendererIndex=3}
   {rendererClass=&NiagaraSpriteRendererProperties, rendererIndex=4}
  ]
 }
 Intro {
  emitterSpawnScript [&InitializeAudioPlayer(Audio Player=DataInterface() {propertyValues {
      Attenuation=null
      Concurrency=null
      MaxPlaysPerTick=10
      SoundToPlay=/CosmeticCompanions/Assets/Audio_Shared/Interaction/Petting/SC_Mimosa_Interaction_React_Default
      bAllowLoopingOneShotSounds=false
      bLimitPlaysPerTick=true
      bOnlyActiveDuringGameplay=false
      bStopWhenComponentIsDestroyed=true
      ConfigurationUserParameter {parameter {name=None, typeDefHandle {registeredTypeIndex=1}, varData []}}
      ParameterNames []
     }})]
  emitterUpdateScript [
   &EmitterState(Life Cycle Mode=::NewEnumerator1, Loop Behavior=::NewEnumerator1, Loop Duration=0.2, UseLoopDelay=false)
   /Niagara/Modules/Audio/PlayAudio() {module=PlayAudio001}
   &SpawnBurst_Instantaneous(Use Spawn Probability=false, UseLoopCountLimit=false)
  ]
  particleSpawnScript [&InitializeParticle(
    Lifetime=&Add_Float(A=$System.IntroLength, B=$System.RestLength)
    Sprite Size Mode=::NewEnumerator3
    Uniform Sprite Size=23
    UsePositionOffset=false
    Write Material Random=true
   )]
  particleUpdateScript [
   &ParticleState()
   &SubUVAnimation(
    Direct Frame Index:NiagaraFloat=&FloatFromCurve(
     FloatCurve=DataInterface() {propertyValues {Curve {keys [<at 0.2 5>,<at 1 9>]}}}
     CurveIndex=&Divide_Float(A=$Particles.Age, B=$System.IntroLength)
    )
    "Sprite Renderer "=DataInterface() {propertyValues {SpriteRenderer=&self:Intro_0.NiagaraSpriteRendererProperties_2}}
    SubUV Animation Mode:ENiagara_SubUVLookupModeV2=::NewEnumerator5
   )
   &OffsetPosition(
    Position Offset=&Lerp_Vector(
     A=$System.HeartIntroSpawnPosition
     B=$System.BurstPosition
     Alpha=&FloatFromCurve(
      FloatCurve=DataInterface() {propertyValues {Curve {keys [
          {arriveTangent=3.2087, interpMode=RCIM_Cubic, leaveTangent=3.2087, tangentMode=RCTM_User, time=0, value=0}
          {arriveTangent=-0.1073, tangentMode=RCTM_User, time=1, value=1}
         ]}}}
      CurveIndex=&Divide_Float(A=$Particles.Age, B=$System.IntroLength)
     )
    )
    Position Offset Coordinate Space:ENiagaraCoordinateSpace=::World
   )
   &ScaleSpriteSize(
    Scale Factor=&Multiply_Vector2D_ByFloat(
     Float=&FloatFromCurve(FloatCurve=DataInterface() {propertyValues {Curve {keys [
          <at 0.75 1 RCIM_Cubic>
          {arriveTangent=-15.6368, interpMode=RCIM_Cubic, leaveTangent=-15.6368, tangentMode=RCTM_User, time=0.8, value=1.2}
          <at 1 0>
         ]}}})
     Vector2D=&Vector2DFromCurve(
      Vector2Curve=DataInterface() {propertyValues {
        XCurve {
         defaultValue=1
         keys [{arriveTangent=1.5094, interpMode=RCIM_Cubic, leaveTangent=1.5094, tangentMode=RCTM_User, time=0, value=0.2}, <at 1 1 RCIM_Cubic>]
        }
        YCurve {keys [
          <at 0 1 RCIM_Cubic>
          {arriveTangent=-0.4796, interpMode=RCIM_Cubic, leaveTangent=-0.4796, tangentMode=RCTM_User, time=0.2, value=1.25}
          <at 1 1>
         ]}
       }}
      CurveIndex=&Divide_Float(A=$Particles.Age, B=$System.IntroLength)
      ScaleCurve=<1 1>
     )
    )
    Scale Sprite Size Mode=::NewEnumerator2
   )
   &ScaleColor(
    Scale Alpha=&FloatFromCurve(FloatCurve=DataInterface() {propertyValues {Curve {
        defaultValue=0
        keys [
         {arriveTangent=9.0725, interpMode=RCIM_Cubic, leaveTangent=9.0725, tangentMode=RCTM_User, time=0, value=0}
         {arriveTangent=-0.2441, leaveTangent=-0.2441, time=0.15, value=1}
         {arriveTangent=-1.146, leaveTangent=-1.146, time=0.97, value=1}
         {arriveTangent=-0.0405, interpMode=RCIM_Cubic, leaveTangent=-0.0405, tangentMode=RCTM_User, time=0.971, value=1}
        ]
       }}})
    Scale RGB=&VectorFromFloat(Value=&FNTimeOfDayScalar(Night Scale=0.65, Morning Scale=0.75, Evening Scale=0.9))
    ScaleA=true
    ScaleRGB=true
   )
   &FN_Sprite_Attributes(Color=<rgba 1 0.4945 0 2>, Create Color Attribute=true, Module Output:ENiagara_AttributeSamplingApplyOutput=::NewEnumerator1)
   &SolveForcesAndVelocity(Clamp Velocity=false, Limit Acceleration=false)
  ]
  rendererClasses [&NiagaraSpriteRendererProperties]
  renderers [
   {rendererClass=&NiagaraSpriteRendererProperties}
   {rendererClass=&NiagaraSpriteRendererProperties, rendererIndex=1}
  ]
 }
 PixieDust_Trail_Packed {
  emitterUpdateScript [
   &EmitterState(Life Cycle Mode=::NewEnumerator1, Loop Behavior=::NewEnumerator1, Loop Duration=0.5, UseLoopDelay=false)
   &SpawnParticlesFromOtherEmitter(
    Attribute Reader=DataInterface() {propertyValues {EmitterBinding {bindingMode=Other, emitterName=Intro}}}
    Spawn Rate=&FN_PickFloatDependOnPlatform(TargetPlatformActive=30, TargetPlatformNoActive=15)
    Spawning Enabled:NiagaraBool=&SetBoolByFloatComparison(
     A=&FloatFromCurve(FloatCurve=DataInterface() {propertyValues {Curve {keys [<at 0 1 RCIM_Constant>,<at 0.2 0 RCIM_Constant>]}}}, CurveIndex=$Emitter.Age)
     B=0
    )
    Use Maximum Age=false
    Use Minimum Age=false
   )
   &SpawnParticlesFromOtherEmitter(
    Attribute Reader=DataInterface() {propertyValues {EmitterBinding {bindingMode=Other, emitterName=Intro}}}
    Spawn Group=1
    Spawn Rate=&FN_PickFloatDependOnPlatform(TargetPlatformActive=25, TargetPlatformNoActive=13)
    Spawning Enabled:NiagaraBool=&SetBoolByFloatComparison(
     A=&FloatFromCurve(
      FloatCurve=DataInterface() {propertyValues {Curve {keys [<at 0 0 RCIM_Constant>,<at 0.2 1 RCIM_Constant>,<at 0.5 0 RCIM_Constant>]}}}
      CurveIndex=$Emitter.Age
     )
     B=0
    )
    Use Maximum Age=false
    Use Minimum Age=false
   ) {module=SpawnParticlesFromOtherEmitter001}
   &SpawnParticlesFromOtherEmitter(
    Attribute Reader=DataInterface() {propertyValues {EmitterBinding {bindingMode=Other, emitterName=Intro}}}
    Spawn Group=2
    Spawn Rate=&FN_PickFloatDependOnPlatform(TargetPlatformActive=35, TargetPlatformNoActive=15)
    Spawning Enabled:NiagaraBool=&SetBoolByFloatComparison(
     A=&FloatFromCurve(
      FloatCurve=DataInterface() {propertyValues {Curve {keys [<at 0 0 RCIM_Constant>,<at 0.25 1 RCIM_Constant>,<at 0.5 0 RCIM_Constant>]}}}
      CurveIndex=$Emitter.Age
     )
     B=0
    )
    Use Maximum Age=false
    Use Minimum Age=false
   ) {module=SpawnParticlesFromOtherEmitter002}
   /SetVariables(Emitter.EmitterRandom:NiagaraFloat=&RandomRangeFloat(Minimum=0, Maximum=360)) {isSetParameters=true}
  ]
  particleSpawnScript [
   /SetVariables(Particles.VisibilityTag=&MergeHelper_IntSwitch(0=0, 1=0, 2=1, 3=?, 4=?, 5=?, Index=$Emitter.SpawnGroup, Num=::NewEnumerator1)) {isSetParameters=true}
   /SetVariables(Particles.SpawnGroup=$Emitter.SpawnGroup) {isSetParameters=true}
   &InitializeParticle(
    AdjustAlpha=true
    AdjustHue=true
    AdjustSaturation=true
    AdjustValue=true
    Color=<rgba 50 46.6667 30 1>
    Color Mode=::NewEnumerator3
    Hue Shift Range:Vector2f=<-0.05 0.05>
    Lifetime=&MergeHelper_FloatSwitch(
     0=&UniformRangedFloat(Minimum=0.8, Maximum=1)
     1=&UniformRangedFloat(Minimum=0.4, Maximum=0.7)
     2=?
     3=?
     Index=$Particles.VisibilityTag
     Num=::NewEnumerator0
    )
    Position Offset=&MergeHelper_VectorSwitch(0=<0 0 0>, 1=$System.BurstPosition, 2=?, 3=?, Index=$Emitter.SpawnGroup, Num=::NewEnumerator0)
    Position Offset Coordinate Space=::Local
    Sprite Rotation Angle=&MergeHelper_FloatSwitch(
     0=&RandomRangeFloat(Minimum=0, Maximum=1)
     1=&Add_Float(
      A=&Multiply_Float(
       A=72
       B=&MakeFloatFromInt(Int=/CascadeToNiagaraConverter/NiagaraScripts/DynamicInputScripts/SequentialInt(Use Array Data:NiagaraBool=false, First Number:NiagaraInt32=0, Sequence:NiagaraDataInterfaceArrayInt32=?, Last Number:NiagaraInt32=5))
      )
      B=$Emitter.EmitterRandom
     )
     2=?
     3=?
     Index=$Particles.VisibilityTag
     Num=::NewEnumerator0
    )
    Sprite Rotation Mode=::NewEnumerator0
    Sprite Size Mode=::NewEnumerator3
    Sprite UV Mode=::NewEnumerator4
    Sprite UV Scale:Vector2f=&MergeHelper_Vector2Switch(
     0=&MakeVector2D(
      X=&UniformAOrBFloat(A=-1, Evaluation Type=::NewEnumerator0, Recalculate Random Each Loop=true, Override Seed=false)
      Y=&UniformAOrBFloat(A=-1, Evaluation Type=::NewEnumerator0, Recalculate Random Each Loop=true, Override Seed=false)
     )
     1=&MakeVector2D(
      X=&UniformAOrBFloat(A=-1, Evaluation Type=::NewEnumerator0, Recalculate Random Each Loop=true, Override Seed=false)
      Y=1
     )
     2=?
     3=?
     4=?
     Index=$Particles.VisibilityTag
     Num=::NewEnumerator0
    )
    Uniform Sprite Size=&MergeHelper_FloatSwitch(
     0=&RandomRangeFloat(Minimum=18.5, Maximum=25)
     1=&RandomRangeFloat(Minimum=28, Maximum=35)
     2=?
     3=?
     Index=$Emitter.SpawnGroup
     Num=::NewEnumerator0
    )
    UsePositionOffset=true
    Value Range:Vector2f=<1 1.3>
    Write Material Random=true
   )
   &SampleParticlesFromOtherEmitter()
   &DynamicMaterialParameters(
    Index 0 Param 1=&RandomRangeFloat(Minimum=-0.05, Maximum=0.05)
    Index 0 Param 2=&RandomRangeFloat(Minimum=0, Maximum=32)
    Module.Index0Float0Write=true
    Module.Index0Float1Write=true
    Module.Index0Float2Write=true
    Module.Index0Float3Write=true
   )
   &AddVelocity(
    Rotation Coordinate Space=::World
    Velocity=&MergeHelper_VectorSwitch(
     0=&MakeVector(X=0, Y=0, Z=&RandomRangeFloat(Minimum=25, Maximum=50))
     1=<0 0 0>
     2=<0 0 0>
     3=?
     Index=$Emitter.SpawnGroup
     Num=::NewEnumerator1
    )
   )
   &CameraOffset(Camera Offset Amount=&MergeHelper_FloatSwitch(0=5, 1=-5, 2=?, 3=?, Index=$Emitter.SpawnGroup, Num=::NewEnumerator0))
   &SubUVAnimation(
    End Frame Range Override=3
    Evaluation Type=::NewEnumerator0
    Output Alternate Renderer Binding=true
    "Sprite Renderer "=DataInterface() {propertyValues {SpriteRenderer=&self:PixieDust_Trail_Packed_0.NiagaraSpriteRendererProperties_0}}
    SubUV Animation Mode:ENiagara_SubUVLookupModeV2=::NewEnumerator2
    UseEndFrame=true
    UseStartFrame=true
   )
  ]
  particleUpdateScript [
   &ParticleState()
   &Drag(Drag=3, Use Linear Drag=true, Use Rotational Drag=true)
   &ScaleSpriteSize(Uniform Curve Sprite Scale=DataInterface() {propertyValues {Curve {keys [
        {arriveTangent=2.2332, interpMode=RCIM_Cubic, leaveTangent=2.2332, tangentMode=RCTM_User, time=0, value=0.6}
        {tangentMode=RCTM_User, time=1, value=1.25}
       ]}}})
   &ScaleColor(
    Scale Alpha=&MergeHelper_FloatSwitch(
     0=&FloatFromCurve(
      FloatCurve=DataInterface() {propertyValues {Curve {keys [
          {interpMode=RCIM_Cubic, leaveTangentWeight=0.0333, tangentMode=RCTM_User, time=0, value=0}
          {arriveTangentWeight=0.0333, interpMode=RCIM_Cubic, leaveTangentWeight=0.3, tangentMode=RCTM_User, time=0.1, value=1}
          {arriveTangentWeight=0.3, interpMode=RCIM_Cubic, tangentMode=RCTM_User, time=1, value=0}
         ]}}}
      Scale Curve=10
     )
     1=&Multiply_Float(
      A=&FloatFromCurve(
       FloatCurve=DataInterface() {propertyValues {Curve {keys [
           {interpMode=RCIM_Cubic, leaveTangentWeight=0.0333, tangentMode=RCTM_User, time=0, value=0}
           {arriveTangentWeight=0.0333, interpMode=RCIM_Cubic, leaveTangentWeight=0.3, tangentMode=RCTM_User, time=0.3, value=1}
           {arriveTangentWeight=0.3, interpMode=RCIM_Cubic, tangentMode=RCTM_User, time=1, value=0}
          ]}}}
       Scale Curve=5
      )
      B=&FloatFromCurve(FloatCurve=DataInterface() {propertyValues {Curve {keys [<at 0.5 1 RCIM_Cubic>,<at 1.5 0 RCIM_Cubic>]}}}, CurveIndex=$Engine.System.Age)
     )
     2=?
     3=?
     Index=$Particles.SpawnGroup
     Num=::NewEnumerator0
    )
    ScaleA=true
    ScaleRGB=true
   )
   &AccelerationForce(Acceleration=&MergeHelper_VectorSwitch(0=<0 0 0>, 1=<0 0 -10>, 2=?, 3=?, Index=$Particles.SpawnGroup, Num=::NewEnumerator0))
   &FN_Sprite_Attributes(
    Color=&MakeLinearColorFromVectorAndFloat(
     "Vector (RGB)"=&MakeVectorFromLinearColorRGB(Color=&ColorFromCurve(ColorCurve=DataInterface() {propertyValues {
         AlphaCurve {keys [<at 0 1 RCIM_Cubic>]}
         BlueCurve {keys [<at 0.2285 0.9 RCIM_Cubic>,<at 0.3986 0.55 RCIM_Cubic>]}
         GreenCurve {keys [<at 0.2285 2.65 RCIM_Cubic>,<at 0.3986 0.8875 RCIM_Cubic>]}
         RedCurve {keys [<at 0.2285 3 RCIM_Cubic>,<at 0.3986 1 RCIM_Cubic>]}
        }}))
     "Float (Alpha)"=&Multiply_Float(
      A=&FloatFromCurve(FloatCurve=DataInterface() {propertyValues {Curve {keys [<at 0 1 RCIM_Cubic>,<at 1 0 RCIM_Cubic>]}}}, Scale Curve=$System.ToDScalar)
      B=&FloatFromCurve(FloatCurve=DataInterface() {propertyValues {Curve {keys [<at 0.5 1 RCIM_Cubic>,<at 1.5 0 RCIM_Cubic>]}}}, CurveIndex=$Engine.System.Age)
     )
    )
    Create Camera Offset Attribute=true
    Create Color Attribute=true
    Create SpriteSize Attribute=true
    Module Output:ENiagara_AttributeSamplingApplyOutput=::NewEnumerator1
    SpriteSize=&Multiply_Vector2D_ByFloat(
     Float=&FloatFromCurve(FloatCurve=DataInterface() {propertyValues {Curve {keys [
          {arriveTangent=2.0035, interpMode=RCIM_Cubic, leaveTangent=2.0035, tangentMode=RCTM_User, time=0, value=0.6}
          {tangentMode=RCTM_User, time=1, value=1.5}
         ]}}})
     Vector2D=&Multiply_Vector2D(
      A=&RandomRangeVector2D(Minimum=<9 15>, Maximum=<13 18.75>)
      B=&Vector2DFromCurve(
       Vector2Curve=DataInterface() {propertyValues {
         XCurve {keys [<at 0.5 1 RCIM_Cubic>,<at 1 0.5 RCIM_Cubic>]}
         YCurve {keys [
           {arriveTangent=1.0106, interpMode=RCIM_Cubic, leaveTangent=1.0106, tangentMode=RCTM_User, time=0, value=1}
           {tangentMode=RCTM_User, time=1, value=1.5}
          ]}
        }}
       ScaleCurve=<1 1>
      )
     )
    )
   )
   &SolveForcesAndVelocity(Clamp Velocity=false, Limit Acceleration=false)
  ]
  rendererClasses [&NiagaraSpriteRendererProperties]
  renderers [
   {rendererClass=&NiagaraSpriteRendererProperties}
   {rendererClass=&NiagaraSpriteRendererProperties, rendererIndex=1}
  ]
 }
 Trail {
  emitterUpdateScript [
   &EmitterState(Life Cycle Mode=::NewEnumerator1, Loop Behavior=::NewEnumerator1, Loop Duration=0.2, UseLoopDelay=false)
   &SpawnParticlesFromOtherEmitter(
    Attribute Reader=DataInterface() {propertyValues {EmitterBinding {bindingMode=Other, emitterName=Intro}}}
    Spawn Rate=100
    Use Maximum Age=false
    Use Minimum Age=false
   )
  ]
  particleSpawnScript [
   &InitializeParticle(Color Mode=::NewEnumerator0, Ribbon Width=15, Ribbon Width Mode=::NewEnumerator1, UsePositionOffset=false, Write Material Random=true)
   &SampleParticlesFromOtherEmitter(Ribbon UV Distance:ENiagara_DirectReadSamplingMode=::NewEnumerator0)
  ]
  particleUpdateScript [
   &ParticleState()
   &DynamicMaterialParameters(
    Index 0 Param 1=-0.1051
    Index 0 Param 3=1
    Index 0 Param 4=1
    Index 1 Param 3=1.4778
    Index 1 Param 4=1
    Module.Index0Float0Write=true
    Module.Index0Float1Write=true
    Module.Index0Float2Write=true
    Module.Index0Float3Write=true
    Module.Index1Float0Write=true
    Module.Index1Float1Write=true
    Module.Index1Float2Write=true
    Module.Index1Float3Write=true
    Write Parameter Index 1=::NewEnumerator1
   )
   &ScaleColor(
    Scale Alpha=&Multiply_Float(
     A=&FloatFromCurve(FloatCurve=DataInterface() {propertyValues {Curve {keys [<at 0 1 RCIM_Cubic>,<at 1 0 RCIM_Cubic>]}}}, Scale Curve=$System.ToDScalar)
     B=&FloatFromCurve(FloatCurve=DataInterface() {propertyValues {Curve {keys [<at 0.15 1 RCIM_Cubic>,<at 0.7 0>]}}}, CurveIndex=$Engine.System.Age)
    )
    Scale RGB=&MakeVectorFromLinearColorRGB(Color=&ColorFromCurve(ColorCurve=DataInterface() {propertyValues {
        AlphaCurve {keys [<at 0 1 RCIM_Cubic>]}
        BlueCurve {keys [<at 0.2285 1.5 RCIM_Cubic>,<at 0.3986 0.8 RCIM_Cubic>]}
        GreenCurve {keys [<at 0.2285 2.75 RCIM_Cubic>,<at 0.3986 0.95 RCIM_Cubic>]}
        RedCurve {keys [<at 0.2285 3 RCIM_Cubic>,<at 0.3986 1 RCIM_Cubic>]}
       }}))
    ScaleA=true
    ScaleRGB=true
   )
   &SolveForcesAndVelocity(Clamp Velocity=false, Limit Acceleration=false)
  ]
  rendererClasses [&NiagaraRibbonRendererProperties]
  renderers [{rendererClass=&NiagaraRibbonRendererProperties}]
 }
 Trail_Spiral {
  emitterUpdateScript [
   &EmitterState(Life Cycle Mode=::NewEnumerator1, Loop Behavior=::NewEnumerator1, Loop Duration=0.2, UseLoopDelay=false)
   &SpawnBurst_Instantaneous(Spawn Count=2, Use Spawn Probability=false, UseLoopCountLimit=false)
   &SpawnRate(
    SpawnRate=&FN_PickFloatDependOnPlatform(TargetPlatformActive=8, TargetPlatformNoActive=6)
    Use Spawn Probability=false
   )
   /SetVariables(Emitter.EmitterRandom:NiagaraFloat=&RandomRangeFloat(Minimum=0, Maximum=1)) {isSetParameters=true}
  ]
  particleSpawnScript [
   &PartitionParticles(Number Of Partitions=4)
   &InitializeParticle(
    Lifetime=0.6
    Mesh Scale Max=<0.4 1.15 1.15>
    Mesh Scale Min=<0.4 0.9 0.9>
    Mesh Scale Mode=::NewEnumerator2
    Position Offset=&AddVectorToPosition(Position=Linked() {linkedVariable {name=System.HeartIntroSpawnPosition, type=Vector3f}}, Vector=<0 0 -20>)
    Position Offset Coordinate Space=::World
    UsePositionOffset=true
    Write Material Random=true
   )
   &InitialMeshOrientation(
    Mesh Orientation Mode=::NewEnumerator7
    Rotation=&Add_Vector(
     A=&MakeVector(X=$Emitter.EmitterRandom, Y=0, Z=0)
     B=&Multiply_Vector(
      A=<0.4 0 0>
      B=&VectorFromFloat(Value=&MakeFloatFromInt(Int=$Particles.PartitionParticles.Partition))
     )
    )
    Use Rotation Vector=true
   )
   &AddVelocity(Rotation Coordinate Space=::World)
  ]
  particleUpdateScript [
   &ParticleState()
   &UpdateMeshOrientation(Rotation Rate=&RandomRangeFloat(Minimum=0.25, Maximum=0.656))
   &ScaleMeshSize(Scale Factor=&MakeVector(
     X=1
     Y=&FloatFromCurve(FloatCurve=DataInterface() {propertyValues {Curve {keys [<at 0 0.5>,<at 0.3 0.85>,<at 0.8 1>,<at 1 1>]}}})
     Z=&FloatFromCurve(FloatCurve=DataInterface() {propertyValues {Curve {keys [<at 0 0.5>,<at 0.3 0.85>,<at 0.8 1>,<at 1 1>]}}})
    ))
   &DynamicMaterialParameters(
    Index 0 Param 2=-0.1
    Index 0 Param 3=1
    Index 0 Param 4=&FloatFromCurve(FloatCurve=DataInterface() {propertyValues {Curve {keys [<at 0 0>,<at 1 1>]}}})
    Index 3 Param 1=500
    Module.Index0Float0Write=true
    Module.Index0Float1Write=true
    Module.Index0Float2Write=true
    Module.Index0Float3Write=true
    Module.Index3Float0Write=true
    Module.Index3Float1Write=true
    Module.Index3Float2Write=true
    Module.Index3Float3Write=true
    Write Parameter Index 3=::NewEnumerator1
   )
   &ScaleColor(
    Scale Alpha=&FloatFromCurve(
     FloatCurve=DataInterface() {propertyValues {Curve {keys [
         {interpMode=RCIM_Cubic, tangentMode=RCTM_User, time=0, value=0}
         {arriveTangent=0.0424, interpMode=RCIM_Cubic, leaveTangent=0.0424, tangentMode=RCTM_User, time=0.1, value=0.8}
         {tangentMode=RCTM_User, time=1, value=0}
        ]}}}
     Scale Curve=$System.ToDScalar
    )
    Scale RGB=&MakeVectorFromLinearColorRGB(Color=&ColorFromCurve(ColorCurve=DataInterface() {propertyValues {
        AlphaCurve {keys []}
        BlueCurve {keys [<at 0.3373 0.65 RCIM_Cubic>,<at 0.6104 0.4 RCIM_Cubic>]}
        GreenCurve {keys [<at 0.3373 0.9417 RCIM_Cubic>,<at 0.6104 0.475 RCIM_Cubic>]}
        RedCurve {keys [<at 0.3373 1 RCIM_Cubic>,<at 0.6104 0.5 RCIM_Cubic>]}
       }}))
    ScaleA=true
    ScaleRGB=true
   )
   &FN_Mesh_Attributes(
    Color=&MakeLinearColorFromVectorAndFloat(
     "Vector (RGB)"=&MakeVectorFromLinearColorRGB(Color=<rgba 2 1.8333 1 1>)
     "Float (Alpha)"=&FloatFromCurve(
      FloatCurve=DataInterface() {propertyValues {Curve {keys [
          {interpMode=RCIM_Cubic, tangentMode=RCTM_User, time=0, value=0}
          {arriveTangent=0.0424, interpMode=RCIM_Cubic, leaveTangent=0.0424, tangentMode=RCTM_User, time=0.1, value=0.8}
          {interpMode=RCIM_Cubic, tangentMode=RCTM_User, time=0.4, value=0.8}
          {tangentMode=RCTM_User, time=1, value=0}
         ]}}}
      Scale Curve=$System.ToDScalar
     )
    )
    Create Color Attribute=true
    Module Output=::NewEnumerator1
   )
   &Drag(Drag=3, Use Linear Drag=true, Use Rotational Drag=true)
   &AccelerationForce(Acceleration=<0 0 35>)
   &SolveForcesAndVelocity(Clamp Velocity=false, Limit Acceleration=false)
  ]
  rendererClasses [&NiagaraMeshRendererProperties]
  renderers [
   {rendererClass=&NiagaraMeshRendererProperties, rendererIndex=0}
   {rendererClass=&NiagaraMeshRendererProperties, rendererIndex=1}
  ]
 }
}
summary {
 systemName=NS_ReactFX_MinkCalico
 emitters [
  {emitterName=Intro, rendererClasses [&NiagaraSpriteRendererProperties]}
  {emitterName=Trail, rendererClasses [&NiagaraRibbonRendererProperties]}
  {emitterName=Trail_Spiral, rendererClasses [&NiagaraMeshRendererProperties]}
  {emitterName=PixieDust_Trail_Packed, rendererClasses [&NiagaraSpriteRendererProperties]}
  {emitterName=Burst_Packed, rendererClasses [&NiagaraSpriteRendererProperties, &NiagaraMeshRendererProperties]}
 ]
 userVariables [
  User.bIsFrontEnd {type=NiagaraBool}
  User.bIsFrontEndPreview {type=NiagaraBool}
 ]
}
