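# Niagara reaction effect for cosmetic companions: a timed sprite vignette in which a bandage girl appears with a sound and a doctor hat swoops in.
# Pop, smoke and mesh wind bursts accent the beats, while hat and heart ribbons trace the motion.
# Emitters share hand-keyed age curves for paths, wobble and frame swaps, offset by System.BurstPosition.
# Most emitters use local GimbalLock modules fed by camera vectors.
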NiagaraSystem /CosmeticCompanions/FX/React/NS_ReactFX_MummyPast
dependencies {
 usedDataInterfaces [
  &NiagaraDataInterfaceCurve
  &NiagaraDataInterfaceVectorCurve
  &NiagaraDataInterfaceArrayFloat
  &NiagaraDataInterfaceSpriteRendererInfo
  &NiagaraDataInterfaceArrayInt32
  &NiagaraDataInterfaceCamera
  &NiagaraDataInterfaceAudioPlayer
 ]
 usedDynamicInputs [&FN_FrontEndVector3Modifier]
 usedModules [
  &self:SystemSpawnScript.SystemScriptSource.SystemScriptGraph.NiagaraNodeAssignment_0.SetVariables_#1
  &self:SystemSpawnScript.SystemScriptSource.SystemScriptGraph.NiagaraNodeAssignment_1.SetVariables_#2
  &self:SystemSpawnScript.SystemScriptSource.SystemScriptGraph.NiagaraNodeAssignment_4.SetVariables_#3
  &self:SystemSpawnScript.SystemScriptSource.SystemScriptGraph.NiagaraNodeAssignment_7.SetVariables_#4
  &SystemState
 ]
}
emitters {
 BandageGirl {
  emitterSpawnScript [
   &InitializeAudioPlayer(Audio Player=DataInterface() {propertyValues {
      Attenuation=null
      Concurrency=null
      MaxPlaysPerTick=10
      SoundToPlay=/CosmeticCompanions/Assets/Audio_Shared/Interaction/Petting/SC_Mimosa_Interaction_React_Default
      bAllowLoopingOneShotSounds=false
      bLimitPlaysPerTick=true
      bOnlyActiveDuringGameplay=false
      bStopWhenComponentIsDestroyed=true
      ConfigurationUserParameter {parameter {name=None, typeDefHandle {registeredTypeIndex=1}, varData []}}
      ParameterNames []
     }})
   /Niagara/Modules/Audio/PlayAudio() {module=PlayAudio001}
  ]
  emitterUpdateScript [
   &EmitterState()
   &SpawnBurst_Instantaneous(Use Spawn Probability=false, UseLoopCountLimit=false)
  ]
  particleSpawnScript [
   &InitializeParticle(
    Color=<rgba 0.9427 0.7512 0.836 1>
    Lifetime=4
    Sprite Rotation Mode=::NewEnumerator0
    Sprite Size Mode=::NewEnumerator3
    Uniform Sprite Size=28.5
    UsePositionOffset=false
    Write Material Random=true
   )
   /SetVariables(Particles.SortingLayer=5) {isSetParameters=true}
  ]
  particleUpdateScript [
   &ParticleState()
   &ScaleSpriteSize(Uniform Curve Sprite Scale=DataInterface() {propertyValues {Curve {keys [{tangentMode=RCTM_User, time=0, value=1.4}, {tangentMode=RCTM_User, time=0.05, value=1}]}}}) {module=ScaleSpriteSize001}
   &ScaleColor(
    Scale Alpha=&FloatFromCurve(FloatCurve=DataInterface() {propertyValues {Curve {
        defaultValue=0
        keys [
         {arriveTangent=-0.2441, time=0.014, value=0.9847}
         {arriveTangent=-1.146, leaveTangent=-1.146, time=0.95, value=1}
         {arriveTangent=-0.0405, interpMode=RCIM_Cubic, leaveTangent=-0.0405, tangentMode=RCTM_User, time=1, value=0}
        ]
       }}})
    ScaleA=true
    ScaleRGB=true
   )
   &ScaleSpriteSize(
    Uniform Curve Index=$Emitter.Age
    Uniform Curve Sprite Scale=DataInterface() {propertyValues {Curve {keys [
        {arriveTangent=0.801, interpMode=RCIM_Cubic, leaveTangent=0.801, tangentMode=RCTM_User, time=1.1221, value=1}
        <at 1.1946 1.5 RCIM_Cubic>
        {arriveTangent=-0.5527, interpMode=RCIM_Cubic, leaveTangent=-0.5527, tangentMode=RCTM_User, time=1.2928, value=1}
        {arriveTangent=-1.1277, interpMode=RCIM_Cubic, leaveTangent=-1.1277, tangentMode=RCTM_User, time=1.3229, value=1}
        {arriveTangent=-1.3672, interpMode=RCIM_Cubic, leaveTangent=-1.3672, tangentMode=RCTM_User, time=1.4835, value=0.5}
        <at 1.4995 0 RCIM_Cubic>
       ]}}}
   )
   /SetVariables(
    Particles.Glow Color=&Multiply_LinearColorByFloat(Linear Color=$System.Layer 2 Color, Float=0.3)
    Particles.Glow Size=&Multiply_Vector2D(A=$Particles.SpriteSize, B=<0.6 0.6>)
   ) {isSetParameters=true}
   &SubUVAnimation(
    Direct Frame Index:NiagaraFloat=&FloatFromCurve(FloatCurve=DataInterface() {propertyValues {Curve {keys [<at 0.888 1>,<at 0.988 2>]}}}, CurveIndex=$Emitter.Age)
    "Sprite Renderer "=DataInterface() {propertyValues {SpriteRenderer=&self:BandageGirl_0.NiagaraSpriteRendererProperties_2}}
    SubUV Animation Mode:ENiagara_SubUVLookupModeV2=::NewEnumerator5
   )
   &OffsetPosition(Position Offset=&MakeVector(
     X=0
     Y=&FloatFromCurve(
      FloatCurve=DataInterface() {propertyValues {Curve {keys [
          {arriveTangent=-391.5863, interpMode=RCIM_Cubic, leaveTangent=-391.5863, tangentMode=RCTM_User, time=0.7075, value=0}
          {arriveTangent=-179.2552, interpMode=RCIM_Cubic, leaveTangent=-179.2552, tangentMode=RCTM_User, time=0.84, value=-25}
         ]}}}
      CurveIndex=$Emitter.NormalizedLoopAge
     )
     Z=&FloatFromCurve(
      FloatCurve=DataInterface() {propertyValues {Curve {keys [<at 0.7075 0 RCIM_Cubic>, {arriveTangent=830.4605, tangentMode=RCTM_User, time=0.84, value=35}]}}}
      CurveIndex=$Emitter.NormalizedLoopAge
     )
    ))
   &OffsetPosition(Position Offset=&MakeVector(
     X=0
     Y=&Lerp_Float(
      A=0
      B=&Waveform("Waveform [1]"=::NewEnumerator0, "[1] Frequency"=10, "[1] Amplitude Min/Max"=<-1 1>, "[1] Scale"=10)
      Alpha=&FloatFromCurve(
       FloatCurve=DataInterface() {propertyValues {Curve {keys [
           <at 0 0>
           {arriveTangent=0.1654, interpMode=RCIM_Cubic, leaveTangent=0.1654, time=0.6348, value=0}
           {arriveTangent=0.1124, interpMode=RCIM_Cubic, leaveTangent=0.1124, time=0.8348, value=0.138}
           {arriveTangent=0.0151, interpMode=RCIM_Cubic, leaveTangent=0.0151, tangentMode=RCTM_User, time=1.304, value=0.0752}
           {arriveTangent=0.0783, interpMode=RCIM_Cubic, leaveTangent=0.0783, tangentMode=RCTM_User, time=2.7262, value=0.062}
           <at 3.0153 0.1 RCIM_Cubic>
          ]}}}
       CurveIndex=$Emitter.Age
      )
     )
     Z=0
    )) {module=OffsetPosition001}
   &OffsetPosition(Position Offset=$System.BurstPosition) {module=OffsetPosition002}
   &self:GimbalLock_06(Gimbal Lock Direction:NiagaraPosition=&ReturnCameraVector(Camera Vectors=::NewEnumerator4, Coordinate Space=::Local, Origin Position=$Engine.Emitter.SimulationPosition)) {module=GimbalLock_06}
  ]
  rendererClasses [&NiagaraSpriteRendererProperties]
  renderers [
   {rendererClass=&NiagaraSpriteRendererProperties}
   {rendererClass=&NiagaraSpriteRendererProperties, rendererIndex=1}
  ]
 }
 Bottom_Wind {
  emitterUpdateScript [
   &EmitterState(Life Cycle Mode=::NewEnumerator1, Loop Behavior=::NewEnumerator1, Loop Duration=1.75, UseLoopDelay=false)
   &SpawnRate(
    SpawnRate=&FloatFromCurve(
     FloatCurve=DataInterface() {propertyValues {Curve {keys [
         <at 0 0>
         {arriveTangent=2.0106, interpMode=RCIM_Cubic, leaveTangent=2.0106, time=0.4255, value=0}
         {arriveTangent=0.0151, interpMode=RCIM_Cubic, leaveTangent=0.0151, tangentMode=RCTM_User, time=0.4974, value=1}
         {arriveTangent=-3.1696, interpMode=RCIM_Cubic, leaveTangent=-3.1696, time=0.6943, value=1}
         <at 0.8129 0 RCIM_Cubic>
        ]}}}
     CurveIndex=$Emitter.NormalizedLoopAge
     Scale Curve=45
    )
    Use Spawn Probability=false
   )
   &SpawnBurst_Instantaneous(Use Spawn Probability=false, UseLoopCountLimit=false) {enabled=false}
  ]
  particleSpawnScript [
   &InitializeParticle(
    Color=<rgba 1 1 1 0.3>
    Lifetime=&RandomRangeFloat(Minimum=0.27, Maximum=0.5)
    Mesh Scale Max=<0.3 0.4 0.2>
    Mesh Scale Min=<0.1 0.2 0.2>
    Mesh Scale Mode=::NewEnumerator2
    Sprite Rotation Mode=::NewEnumerator2
    UsePositionOffset=true
    Write Material Random=true
   )
   &ShapeLocation(
    MaskSpawnGroup=false
    Non Uniform Scale=<0 1.5 1>
    Offset=<0 -28 40>
    Offset Mode=::NewEnumerator0
    Override Seed=false
    Sphere Radius=3
    Sphere Surface Distribution=0.6905
    UseOwnerScale=true
   )
   /SetVariables(Particles.VisibilityTag=&RandomRangeInt(Minimum=0, Maximum=3)) {isSetParameters=true}
   &InitialMeshOrientation(Mesh Orientation Mode=::NewEnumerator2, Orientation Vector=<0 -0.3 1>, Use Rotation Vector=false)
  ]
  particleUpdateScript [
   &ParticleState()
   &ScaleMeshSize()
   &ScaleColor(
    Scale Alpha=&FloatFromCurve(FloatCurve=DataInterface() {propertyValues {Curve {keys [<at 0.4 1>,<at 0.6 0>]}}})
    ScaleA=true
    ScaleRGB=true
   ) {enabled=false}
   &DynamicMaterialParameters(
    Index 0 Param 1=2
    Index 0 Param 3=&FloatFromCurve(FloatCurve=DataInterface() {propertyValues {Curve {keys [
         {arriveTangent=-3.8358, interpMode=RCIM_Cubic, leaveTangent=-3.8358, tangentMode=RCTM_User, time=0, value=1}
         {arriveTangent=-2.3408, interpMode=RCIM_Cubic, leaveTangent=-2.3408, tangentMode=RCTM_User, time=1, value=-2}
        ]}}})
    Module.Index0Float0Write=true
    Module.Index0Float1Write=false
    Module.Index0Float2Write=true
    Module.Index0Float3Write=false
   )
   &OffsetPosition(Position Offset=$System.BurstPosition) {module=OffsetPosition002}
   &self:GimbalLock_01(Gimbal Lock Direction:NiagaraPosition=&ReturnCameraVector(Camera Vectors=::NewEnumerator4, Coordinate Space=::Local, Origin Position=$Engine.Emitter.SimulationPosition)) {module=GimbalLock_01}
   &SkeletalMeshLocation(Bone Sampling Mode=::NewEnumerator4, MaskSpawnGroup=false, Sampled Position Offset=<-50 0 30>, Skeletal Mesh=$System.BodySkel) {enabled=false}
  ]
  rendererClasses [&NiagaraMeshRendererProperties]
  renderers [
   {rendererClass=&NiagaraMeshRendererProperties, rendererIndex=0}
   {rendererClass=&NiagaraMeshRendererProperties, rendererIndex=1}
   {rendererClass=&NiagaraMeshRendererProperties, rendererIndex=2}
   {rendererClass=&NiagaraMeshRendererProperties, rendererIndex=3}
  ]
 }
 DrHat {
  emitterUpdateScript [
   &EmitterState()
   &SpawnBurst_Instantaneous(
    Spawn Time=&Add_Float(A=$System.IntroLength, B=$System.HatDelay)
    Use Spawn Probability=false
    UseLoopCountLimit=false
   ) {module=SpawnBurst_Instantaneous001}
   &SpawnBurst_Instantaneous(
    Spawn Group=1
    Spawn Time=&Add_Float(A=$System.IntroLength, B=$System.HatDelay)
    Use Spawn Probability=false
    UseLoopCountLimit=false
   )
  ]
  particleSpawnScript [
   &InitializeParticle(
    Lifetime=1.25
    Mesh Scale Mode=::NewEnumerator3
    Mesh Uniform Scale=5
    Ribbon Width=10
    Ribbon Width Mode=::NewEnumerator1
    Sprite Size Mode=::NewEnumerator3
    Uniform Sprite Size=37
    UsePositionOffset=false
    Write Material Random=true
   )
   /SetVariables(Particles.VisibilityTag=$Emitter.SpawnGroup) {isSetParameters=true}
   /SetVariables(Particles.SortingLayer:NiagaraFloat=&SelectFloatFromArray(Direct Array Index=$Particles.VisibilityTag, Float Selection Array=DataInterface() {propertyValues {FloatData [-5,-5]}})) {isSetParameters=true}
   &CameraOffset(Camera Offset Amount=&MaskFloat(Mask Source=1, Mask Parameter=$Particles.VisibilityTag, Mask=0, Masked Float=0.1, Passthrough Float=-0.1))
  ]
  particleUpdateScript [
   &ParticleState(Kill Particles When Lifetime Has Elapsed=false, Loop Particles Lifetime=true)
   /SetVariables(Particles.SpriteSize=&Multiply_Vector2D_ByFloat(
     Float=&FloatFromCurve(FloatCurve=DataInterface() {propertyValues {Curve {
         defaultValue=1
         keys [
          {interpMode=RCIM_Cubic, tangentMode=RCTM_User, time=0, value=0}
          {interpMode=RCIM_Cubic, tangentMode=RCTM_User, time=0.1, value=1}
          {arriveTangent=0.2485, interpMode=RCIM_Cubic, leaveTangent=0.2485, time=0.8864, value=1}
          {arriveTangent=-8.8031, interpMode=RCIM_Cubic, leaveTangent=-8.8031, time=0.9049, value=1.2}
          <at 1 0 RCIM_Cubic>
         ]
        }}})
     Vector2D=$Particles.Initial.SpriteSize
    )) {isSetParameters=true}
   &SubUVAnimation(
    Direct Frame Index:NiagaraFloat=&MakeFloatFromInt(Int=&SelectIntFromArray(Direct Array Index=$Particles.Initial.VisibilityTag, Int Selection Array=DataInterface() {propertyValues {IntData [1,0]}}))
    "Sprite Renderer "=DataInterface() {propertyValues {SpriteRenderer=&self:DrHat_0.NiagaraSpriteRendererProperties_0}}
    SubUV Animation Mode:ENiagara_SubUVLookupModeV2=::NewEnumerator5
   ) {module=SubUVAnimation001}
   &OffsetPosition(Position Offset=&MakeVector(
     X=0
     Y=&FloatFromCurve(
      FloatCurve=DataInterface() {propertyValues {Curve {keys [
          <at 0.2891 -127 RCIM_Cubic>
          {arriveTangent=523.8333, interpMode=RCIM_Cubic, leaveTangent=523.8333, tangentMode=RCTM_User, time=0.4902, value=-25}
         ]}}}
      CurveIndex=$Emitter.NormalizedLoopAge
     )
     Z=&FloatFromCurve(
      FloatCurve=DataInterface() {propertyValues {Curve {keys [<at 0.2839 90>, {arriveTangent=-237.6561, interpMode=RCIM_Cubic, leaveTangent=-237.6561, time=0.4033, value=37.9158}, <at 0.4937 40.1355>]}}}
      CurveIndex=$Emitter.NormalizedLoopAge
     )
    ))
   &OffsetPosition(Position Offset=&MakeVector(
     X=0
     Y=&Lerp_Float(
      A=0
      B=&Waveform("Waveform [1]"=::NewEnumerator0, "[1] Frequency"=10, "[1] Amplitude Min/Max"=<-1 1>, "[1] Scale"=10)
      Alpha=&FloatFromCurve(
       FloatCurve=DataInterface() {propertyValues {Curve {keys [
           <at 0 0>
           {arriveTangent=0.0852, interpMode=RCIM_Cubic, leaveTangent=0.0852, time=0.4255, value=0}
           {arriveTangent=0.2112, interpMode=RCIM_Cubic, leaveTangent=0.2112, time=0.5663, value=0.0482}
           {arriveTangent=0.0151, interpMode=RCIM_Cubic, leaveTangent=0.0151, tangentMode=RCTM_User, time=0.7817, value=0.0752}
          ]}}}
       CurveIndex=$Emitter.NormalizedLoopAge
      )
     )
     Z=0
    )) {module=OffsetPosition001}
   &OffsetPosition(Position Offset=$System.BurstPosition) {module=OffsetPosition002}
   &self:GimbalLock_05(Gimbal Lock Direction:NiagaraPosition=&ReturnCameraVector(Camera Vectors=::NewEnumerator4, Coordinate Space=::Local, Origin Position=$Engine.Emitter.SimulationPosition)) {module=GimbalLock_05}
  ]
  rendererClasses [
   &NiagaraSpriteRendererProperties
   &NiagaraMeshRendererProperties
   &NiagaraRibbonRendererProperties
  ]
  renderers [
   {rendererClass=&NiagaraSpriteRendererProperties}
   {rendererClass=&NiagaraSpriteRendererProperties, rendererIndex=1}
   {rendererClass=&NiagaraMeshRendererProperties, rendererIndex=2}
   {rendererClass=&NiagaraRibbonRendererProperties, rendererIndex=3}
  ]
 }
 DrHatTrail {
  emitterUpdateScript [
   &EmitterState()
   &SpawnRate(
    SpawnGroup=2
    SpawnRate=&FloatFromCurve(
     FloatCurve=DataInterface() {propertyValues {Curve {keys [<at 0.1733 0 RCIM_Constant>,<at 0.2891 1 RCIM_Constant>,<at 0.4902 0 RCIM_Constant>]}}}
     CurveIndex=$Emitter.NormalizedLoopAge
     Scale Curve=80
    )
    Use Spawn Probability=false
   )
  ]
  particleSpawnScript [
   &InitializeParticle(
    Color=&MakeLinearColorFromVectorAndFloat(
     "Vector (RGB)"=<1 1 1>
     "Float (Alpha)"=&FloatFromCurve(
      FloatCurve=DataInterface() {propertyValues {Curve {keys [<at 0 0.6 RCIM_Constant>,<at 0.5586 1 RCIM_Constant>,<at 1 0 RCIM_Constant>]}}}
      CurveIndex=$Emitter.NormalizedLoopAge
     )
    )
    Lifetime=0.2
    Ribbon Width=&FloatFromCurve(
     FloatCurve=DataInterface() {propertyValues {Curve {keys [<at 0 30 RCIM_Constant>,<at 0.5586 15 RCIM_Constant>,<at 1 0 RCIM_Constant>]}}}
     CurveIndex=$Emitter.NormalizedLoopAge
    )
    Ribbon Width Mode=::NewEnumerator1
    UsePositionOffset=false
    Write Material Random=true
   )
   &OffsetPosition(Position Offset=&MakeVector(
     X=0
     Y=&FloatFromCurve(
      FloatCurve=DataInterface() {propertyValues {Curve {keys [
          {interpMode=RCIM_Cubic, tangentMode=RCTM_Break, time=0.2891, value=-127}
          {arriveTangent=523.8333, interpMode=RCIM_Cubic, leaveTangent=25.6421, tangentMode=RCTM_Break, time=0.4902, value=-20}
          {
           arriveTangent=-0.384
           arriveTangentWeight=0.0724
           interpMode=RCIM_Cubic
           leaveTangent=-391.5863
           leaveTangentWeight=0.0442
           tangentMode=RCTM_Break
           time=0.7275
           value=0
          }
          {
           arriveTangent=-179.2552
           arriveTangentWeight=0.1166
           interpMode=RCIM_Cubic
           leaveTangent=-179.2552
           tangentMode=RCTM_Break
           time=0.86
           value=-25
          }
         ]}}}
      CurveIndex=$Emitter.NormalizedLoopAge
     )
     Z=&FloatFromCurve(
      FloatCurve=DataInterface() {propertyValues {Curve {keys [
          <at 0.2839 90>
          {arriveTangent=-237.6561, interpMode=RCIM_Cubic, leaveTangent=-237.6561, time=0.4033, value=37.9158}
          <at 0.4937 40.1355>
          {arriveTangent=-14.0203, interpMode=RCIM_Cubic, leaveTangent=-14.0203, time=0.7275, value=0}
          <at 0.86 35>
         ]}}}
      CurveIndex=$Emitter.NormalizedLoopAge
     )
    ))
   &CameraOffset()
  ]
  particleUpdateScript [
   &ParticleState()
   &OffsetPosition(Position Offset=$System.BurstPosition) {module=OffsetPosition002}
   &self:GimbalLock_04(Gimbal Lock Direction:NiagaraPosition=&ReturnCameraVector(Camera Vectors=::NewEnumerator4, Coordinate Space=::Local, Origin Position=$Engine.Emitter.SimulationPosition)) {module=GimbalLock_04}
   &ScaleColor(
    Scale Alpha=&FloatFromCurve(FloatCurve=DataInterface() {propertyValues {Curve {keys [<at 0 1>,<at 1 0>]}}})
    ScaleA=true
    ScaleRGB=true
   )
  ]
  rendererClasses [&NiagaraRibbonRendererProperties]
  renderers [{rendererClass=&NiagaraRibbonRendererProperties}]
 }
 HeartTrail {
  emitterUpdateScript [
   &EmitterState()
   &SpawnRate(
    SpawnGroup=2
    SpawnRate=&FloatFromCurve(
     FloatCurve=DataInterface() {propertyValues {Curve {keys [<at 0.4902 0 RCIM_Constant>,<at 0.7075 1 RCIM_Constant>,<at 0.8902 0 RCIM_Constant>]}}}
     CurveIndex=$Emitter.NormalizedLoopAge
     Scale Curve=80
    )
    Use Spawn Probability=false
   )
  ]
  particleSpawnScript [
   &InitializeParticle(
    Color=&MakeLinearColorFromVectorAndFloat(
     "Vector (RGB)"=<1 1 1>
     "Float (Alpha)"=&FloatFromCurve(
      FloatCurve=DataInterface() {propertyValues {Curve {keys [<at 0 0.6 RCIM_Constant>,<at 0.5586 1 RCIM_Constant>,<at 1 0 RCIM_Constant>]}}}
      CurveIndex=$Emitter.NormalizedLoopAge
     )
    )
    Lifetime=0.185
    Ribbon Width=&FloatFromCurve(
     FloatCurve=DataInterface() {propertyValues {Curve {keys [
         {
          arriveTangentWeight=0.0563
          interpMode=RCIM_Cubic
          leaveTangent=-14.2447
          leaveTangentWeight=0.0908
          tangentMode=RCTM_Break
          time=0.727
          value=12
         }
         {
          arriveTangent=-185.5073
          arriveTangentWeight=0.0442
          interpMode=RCIM_Cubic
          leaveTangentWeight=0.0467
          tangentMode=RCTM_Break
          time=0.85
          value=7
         }
        ]}}}
     CurveIndex=$Emitter.NormalizedLoopAge
    )
    Ribbon Width Mode=::NewEnumerator1
    UsePositionOffset=false
    Write Material Random=true
   )
   &OffsetPosition(Position Offset=&MakeVector(
     X=0
     Y=&FloatFromCurve(
      FloatCurve=DataInterface() {propertyValues {Curve {keys [
          {interpMode=RCIM_Cubic, tangentMode=RCTM_Break, time=0.2891, value=-127}
          {arriveTangent=523.8333, interpMode=RCIM_Cubic, leaveTangent=25.6421, tangentMode=RCTM_Break, time=0.4902, value=-20}
          {
           arriveTangent=-0.384
           arriveTangentWeight=0.0724
           interpMode=RCIM_Cubic
           leaveTangent=-391.5863
           leaveTangentWeight=0.0442
           tangentMode=RCTM_Break
           time=0.7275
           value=0
          }
          {
           arriveTangent=-179.2552
           arriveTangentWeight=0.1166
           interpMode=RCIM_Cubic
           leaveTangent=-179.2552
           tangentMode=RCTM_Break
           time=0.86
           value=-25
          }
         ]}}}
      CurveIndex=$Emitter.NormalizedLoopAge
     )
     Z=&FloatFromCurve(
      FloatCurve=DataInterface() {propertyValues {Curve {keys [
          <at 0.2839 90>
          {arriveTangent=-237.6561, interpMode=RCIM_Cubic, leaveTangent=-237.6561, time=0.4033, value=37.9158}
          <at 0.4937 40.1355>
          {arriveTangent=-14.0203, interpMode=RCIM_Cubic, leaveTangent=-14.0203, time=0.7275, value=0}
          <at 0.86 35>
         ]}}}
      CurveIndex=$Emitter.NormalizedLoopAge
     )
    ))
   &CameraOffset()
  ]
  particleUpdateScript [
   &ParticleState()
   &OffsetPosition(Position Offset=$System.BurstPosition) {module=OffsetPosition002}
   &self:GimbalLock_03(Gimbal Lock Direction:NiagaraPosition=&ReturnCameraVector(Camera Vectors=::NewEnumerator4, Coordinate Space=::Local, Origin Position=$Engine.Emitter.SimulationPosition)) {module=GimbalLock_03}
   &ScaleColor(
    Scale Alpha=&FloatFromCurve(FloatCurve=DataInterface() {propertyValues {Curve {keys [<at 0 1>,<at 1 0>]}}})
    ScaleA=true
    ScaleRGB=true
   )
  ]
  rendererClasses [&NiagaraRibbonRendererProperties]
  renderers [{rendererClass=&NiagaraRibbonRendererProperties}]
 }
 Pop {
  emitterUpdateScript [
   &EmitterState()
   &SpawnBurst_Instantaneous(Spawn Time=&Subtract_Float(A=$System.IntroLength, B=0), Use Spawn Probability=false, UseLoopCountLimit=false)
  ]
  particleSpawnScript [
   &InitializeParticle(
    Lifetime=0.2
    Position Offset=$System.BurstPosition
    Sprite Rotation Mode=::NewEnumerator0
    Sprite Size Mode=::NewEnumerator3
    Uniform Sprite Size=36
    UsePositionOffset=true
    Write Material Random=true
   )
   &DynamicMaterialParameters(
    Index 0 Param 4=1
    Module.Index0Float0Write=true
    Module.Index0Float1Write=true
    Module.Index0Float2Write=true
    Module.Index0Float3Write=true
   )
  ]
  particleUpdateScript [
   &ParticleState()
   &ScaleSpriteSize(
    Scale Sprite Size Mode=::NewEnumerator0
    Uniform Scale Factor=&FloatFromCurve(FloatCurve=DataInterface() {propertyValues {Curve {keys [
         {arriveTangent=3.3458, interpMode=RCIM_Cubic, leaveTangent=3.3458, tangentMode=RCTM_User, time=0, value=1}
         {arriveTangent=1.251, tangentMode=RCTM_User, time=1, value=3}
        ]}}})
   ) {module=ScaleSpriteSize002}
   &ScaleColor(
    Scale Alpha=&FloatFromCurve(FloatCurve=DataInterface() {propertyValues {Curve {
        defaultValue=0
        keys [
         {arriveTangent=0.3137, interpMode=RCIM_Cubic, leaveTangent=0.3137, tangentMode=RCTM_User, time=0, value=1}
         {arriveTangent=-2.0958, interpMode=RCIM_Cubic, leaveTangent=-2.0958, tangentMode=RCTM_User, time=1, value=0}
        ]
       }}})
    Scale RGB=&VectorFromCurve(VectorCurve=DataInterface() {propertyValues {
       XCurve {keys [{interpMode=RCIM_Cubic, tangentMode=RCTM_User, time=0, value=1}, {arriveTangent=-0.2294, tangentMode=RCTM_User, time=1, value=0}]}
       YCurve {keys [{interpMode=RCIM_Cubic, tangentMode=RCTM_User, time=0, value=1}, {arriveTangent=-0.2294, tangentMode=RCTM_User, time=1, value=0}]}
       ZCurve {keys [{interpMode=RCIM_Cubic, tangentMode=RCTM_User, time=0, value=1}, {arriveTangent=-0.2294, tangentMode=RCTM_User, time=1, value=0}]}
      }})
    ScaleA=true
    ScaleRGB=true
   )
   &SolveForcesAndVelocity(Clamp Velocity=false, Limit Acceleration=false)
   /SetVariables(Particles.Glow Color=&Multiply_LinearColorByFloat(
     Linear Color=&MultiplyLinearColor(
      Linear Color=$Particles.Color
      Scale Factor=&MakeVector4_FromLinearColor(LinearColor=<rgba 1 0.5877 0.8563 1>)
     )
     Float=0.33
    )) {isSetParameters=true}
  ]
  rendererClasses [&NiagaraSpriteRendererProperties]
  renderers [
   {rendererClass=&NiagaraSpriteRendererProperties}
   {rendererClass=&NiagaraSpriteRendererProperties, rendererIndex=1}
  ]
 }
 SmokeSpawn {
  emitterUpdateScript [
   &EmitterState()
   &SpawnBurst_Instantaneous(
    Spawn Time=&Add_Float(A=$System.IntroLength, B=$System.HatDelay)
    Use Spawn Probability=false
    UseLoopCountLimit=false
   )
  ]
  particleSpawnScript [
   &InitializeParticle(
    Color=&FNTimeOfDayColor(
     Night Color=<rgba 0.2149 0.3102 0.842 1>
     Morning Color=<rgba 0.7078 0.5363 0.5148 1>
     Day Color=<rgba 0.8096 0.8096 0.8096 1>
     Evening Color=<rgba 0.6076 0.4758 0.7031 1>
    )
    Lifetime=0.34
    Sprite Rotation Angle=&RandomRangeFloat(Minimum=-45, Maximum=45)
    Sprite Rotation Mode=::NewEnumerator0
    Sprite Size Mode=::NewEnumerator3
    Uniform Sprite Size=50
    UsePositionOffset=false
    Write Material Random=true
   ) {module=InitializeParticle001}
   &CameraOffset(Camera Offset Amount=20)
   &DynamicMaterialParameters(
    Index 0 Param 2=&UniformAOrBFloat(A=0)
    Module.Param0WriteEnabled=true
    Module.Param1WriteEnabled=true
    Module.Param2WriteEnabled=true
    Module.Param3WriteEnabled=true
   ) {module=DynamicMaterialParameters001}
  ]
  particleUpdateScript [
   &ParticleState()
   &SolveForcesAndVelocity(Clamp Velocity=false, Limit Acceleration=false)
   &DynamicMaterialParameters(
    Index 0 Param 1=&FloatFromCurve(FloatCurve=DataInterface() {propertyValues {Curve {keys [<at 0 0>,<at 1 1>]}}})
    Index 0 Param 2=1
    Module.Param0WriteEnabled=true
    Module.Param1WriteEnabled=true
    Module.Param2WriteEnabled=true
    Module.Param3WriteEnabled=true
   )
   &ScaleColor(
    Scale Alpha=&FloatFromCurve(FloatCurve=DataInterface() {propertyValues {Curve {keys [<at 0 1>,<at 0.8091 0.9932>,<at 1 0>]}}})
    ScaleA=true
    ScaleRGB=true
   )
   &OffsetPosition(Position Offset=&MakeVector(
     X=0
     Y=&FloatFromCurve(FloatCurve=DataInterface() {propertyValues {Curve {keys [<at 0.2891 -127 RCIM_Cubic>]}}}, CurveIndex=$Emitter.NormalizedLoopAge)
     Z=&FloatFromCurve(FloatCurve=DataInterface() {propertyValues {Curve {keys [<at 0.2839 90>]}}}, CurveIndex=$Emitter.NormalizedLoopAge)
    ))
   &OffsetPosition(Position Offset=$System.BurstPosition) {module=OffsetPosition002}
   &self:GimbalLock_02(Gimbal Lock Direction:NiagaraPosition=&ReturnCameraVector(Camera Vectors=::NewEnumerator4, Coordinate Space=::Local, Origin Position=$Engine.Emitter.SimulationPosition)) {module=GimbalLock_02}
  ]
  rendererClasses [&NiagaraSpriteRendererProperties]
  renderers [{rendererClass=&NiagaraSpriteRendererProperties}]
 }
}
summary {
 systemName=NS_ReactFX_MummyPast
 emitters [
  {emitterName=Pop, rendererClasses [&NiagaraSpriteRendererProperties]}
  {
   emitterName=DrHat
   rendererClasses [
    &NiagaraSpriteRendererProperties
    &NiagaraMeshRendererProperties
    &NiagaraRibbonRendererProperties
   ]
  }
  {emitterName=DrHatTrail, rendererClasses [&NiagaraRibbonRendererProperties]}
  {emitterName=BandageGirl, rendererClasses [&NiagaraSpriteRendererProperties]}
  {emitterName=HeartTrail, rendererClasses [&NiagaraRibbonRendererProperties]}
  {emitterName=SmokeSpawn, rendererClasses [&NiagaraSpriteRendererProperties]}
  {emitterName=Bottom_Wind, rendererClasses [&NiagaraMeshRendererProperties]}
 ]
 userVariables [
  User.bIsFrontEnd {type=NiagaraBool}
  User.bIsFrontEndPreview {type=NiagaraBool}
  User.ConcurrentObjectIndex {type=NiagaraFloat, defaultValue=0, description=""}
  User.ObjectIndex {type=NiagaraFloat, defaultValue=0, description=""}
  User.WorkoutSelect {type=NiagaraFloat, defaultValue=0, description=""}
 ]
}
