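# Niagara reaction effect for the NibbleMoss cosmetic companion: a timed intro-and-pop sequence with a petting sound.
# IntroHeart shakes a scaling mesh with curl noise and plays the sound, while MorphSpikes adds green sprites and meshes at a platform-dependent rate.
# DeathBurst and EndImpact follow with delayed sprite and mesh bursts.
# Shared System position, scale and intro-length values drive the emitters, with spawn groups written to VisibilityTag to separate renderer layers.
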NiagaraSystem /CosmeticCompanions/FX/React/NS_ReactFX_NibbleMoss
dependencies {
 usedDataInterfaces [
  &NiagaraDataInterfaceCurve
  &NiagaraDataInterfaceColorCurve
  &NiagaraDataInterfaceSpriteRendererInfo
  &NiagaraDataInterfacePlatformSet
  &NiagaraDataInterfaceVector2DCurve
  &NiagaraDataInterfaceAudioPlayer
 ]
 usedDynamicInputs [
  /Niagara/DynamicInputs/Math/Floor_Float
  &ClampFloat
  &FN_FrontEndVector3Modifier
  &FN_FrontEndFloatModifier
 ]
 usedModules [
  &self:SystemSpawnScript.SystemScriptSource.SystemScriptGraph.NiagaraNodeAssignment_0.SetVariables_#1
  &self:SystemSpawnScript.SystemScriptSource.SystemScriptGraph.NiagaraNodeAssignment_1.SetVariables_#2
  &self:SystemSpawnScript.SystemScriptSource.SystemScriptGraph.NiagaraNodeAssignment_5.SetVariables_#3
  &self:SystemSpawnScript.SystemScriptSource.SystemScriptGraph.NiagaraNodeAssignment_2.SetVariables_#4
  &self:SystemSpawnScript.SystemScriptSource.SystemScriptGraph.NiagaraNodeAssignment_3.SetVariables_#5
  &SystemState
 ]
}
emitters {
 DeathBurst {
  emitterUpdateScript [
   &EmitterState(Life Cycle Mode=::NewEnumerator1, Loop Behavior=::NewEnumerator1, Loop Duration=3, UseLoopDelay=false)
   &SpawnBurst_Instantaneous(
    Spawn Count=3
    Spawn Time=&Subtract_Float(A=$System.IntroLength, B=-1.02)
    Use Spawn Probability=false
    UseLoopCountLimit=false
   )
   &SpawnBurst_Instantaneous(
    Spawn Group=1
    Spawn Time=&Subtract_Float(A=$System.IntroLength, B=-1.12)
    Use Spawn Probability=false
    UseLoopCountLimit=false
   ) {module=SpawnBurst_Instantaneous001}
  ]
  particleSpawnScript [
   /SetVariables(Particles.VisibilityTag=$Emitter.SpawnGroup) {isSetParameters=true}
   &InitializeParticle(
    AdjustAlpha=false
    AdjustHue=true
    AdjustSaturation=false
    AdjustValue=false
    Color=<rgba 0.2597 2 0.0412 1>
    Color Mode=::NewEnumerator3
    Hue Shift Range:Vector2f=<-0.02 0.1>
    Lifetime=&MergeHelper_FloatSwitch(
     0=&RandomRangeFloat(Minimum=0.3, Maximum=0.7)
     1=0.3
     2=?
     3=?
     Index=$Emitter.SpawnGroup
     Num=::NewEnumerator0
    )
    Mesh Scale Mode=::NewEnumerator3
    Mesh Uniform Scale=&Multiply_Float(A=0.38, B=$System.PopScale)
    Position:Vector3f=$System.BurstPosition
    Position Mode=::NewEnumerator1
    Sprite Rotation Mode=::NewEnumerator2
    Sprite Size Mode=::NewEnumerator3
    Uniform Sprite Size=&Multiply_Float(
     A=&RandomRangeFloat(Minimum=30, Maximum=50)
     B=&RandomRangeFloat(Minimum=1, Maximum=2)
    )
    UsePositionOffset=false
    Write Material Random=true
   )
   &CameraOffset(Camera Offset Amount=&RandomRangeFloat(Minimum=10, Maximum=15))
  ]
  particleUpdateScript [
   &ParticleState()
   &ScaleSpriteSize(
    Uniform Curve Scale=$System.PopScale
    Uniform Curve Sprite Scale=DataInterface() {propertyValues {Curve {
       defaultValue=0
       keys [
        {arriveTangent=2.2059, interpMode=RCIM_Cubic, leaveTangent=2.2059, tangentMode=RCTM_User, time=0, value=0.35}
        {interpMode=RCIM_Cubic, leaveTangent=1.402, time=0.9978, value=1.46}
       ]
      }}}
   )
   &ScaleMeshSize(Scale Factor=&VectorFromFloat(Value=&FloatFromCurve(FloatCurve=DataInterface() {propertyValues {Curve {keys [{arriveTangent=1.0626, interpMode=RCIM_Cubic, leaveTangent=1.0626, tangentMode=RCTM_User, time=0, value=0.7}, <at 1 1.2 RCIM_Cubic>]}}})))
   &ScaleColor(
    Scale Alpha=&FloatFromCurve(
     FloatCurve=DataInterface() {propertyValues {Curve {
        defaultValue=0
        keys [
         {arriveTangent=-2.0809, interpMode=RCIM_Cubic, leaveTangent=-2.0809, tangentMode=RCTM_User, time=0.2, value=1}
         {arriveTangent=-0.0404, interpMode=RCIM_Cubic, leaveTangent=-0.0404, tangentMode=RCTM_User, time=1, value=0}
        ]
       }}}
     Scale Curve=&RandomRangeFloat(Minimum=1, Maximum=3)
    )
    Scale RGB=&MakeVectorFromLinearColorRGB(Color=&ColorFromCurve(ColorCurve=DataInterface() {propertyValues {
        AlphaCurve {keys []}
        BlueCurve {keys [<at 0.1264 3>, {interpMode=RCIM_Cubic, leaveTangent=-2.2895, time=0.6387, value=1}, <at 1 1 RCIM_Cubic>]}
        GreenCurve {keys [<at 0.1264 3>, {interpMode=RCIM_Cubic, leaveTangent=-2.2895, time=0.6387, value=1}, <at 1 1 RCIM_Cubic>]}
        RedCurve {keys [<at 0.1264 3>, {interpMode=RCIM_Cubic, leaveTangent=-2.2895, time=0.6387, value=1}, <at 1 1 RCIM_Cubic>]}
       }}))
    ScaleA=true
    ScaleRGB=true
   )
   &Drag(Use Linear Drag=true, Use Rotational Drag=false)
   &GravityForce(Gravity=<0 0 -20>)
   &FN_Mesh_Attributes(
    Color=&MultiplyLinearColor(
     Linear Color=<rgba 0 0 0 1>
     Scale Factor=&MakeVector4FromVectorAndFloat("Vector (RGB)"=<1 1 1>, "Float (Alpha)"=5)
    )
    Create Color Attribute=true
    Create Dynamic Parameter Attribute=true
    Index 0 Param 1:NiagaraFloat=&Lerp_Float(A=0.4, B=1, Alpha=$Particles.NormalizedAge)
    Index 0 Param 2=1
    Module Output=::NewEnumerator1
    Module.Index0Float0Write=true
    Module.Index0Float1Write=true
    Module.Index0Float2Write=true
    Module.Index0Float3Write=true
   )
   &SolveForcesAndVelocity(Clamp Velocity=false, Limit Acceleration=false)
  ]
  rendererClasses [&NiagaraSpriteRendererProperties, &NiagaraMeshRendererProperties]
  renderers [
   {rendererClass=&NiagaraSpriteRendererProperties}
   {rendererClass=&NiagaraMeshRendererProperties, rendererIndex=1}
  ]
 }
 EndImpact {
  emitterUpdateScript [
   &EmitterState(Life Cycle Mode=::NewEnumerator1, Loop Behavior=::NewEnumerator1, Loop Delay=1.55, Loop Duration=0.25, UseLoopDelay=true)
   &SpawnBurst_Instantaneous(Spawn Count=9, Use Spawn Probability=false, UseLoopCountLimit=false)
   &SpawnBurst_Instantaneous(Spawn Count=4, Spawn Group=1, Use Spawn Probability=false, UseLoopCountLimit=false) {module=SpawnBurst_Instantaneous001}
   &SpawnBurst_Instantaneous(Spawn Group=2, Use Spawn Probability=false, UseLoopCountLimit=false) {module=SpawnBurst_Instantaneous002}
   /SetVariables(Emitter.DistortionStrength:NiagaraFloat=&Power(a=&Lerp_Float(A=1.5, B=0, Alpha=$Emitter.NormalizedLoopAge), b=1)) {isSetParameters=true}
  ]
  particleSpawnScript [
   /SetVariables(
    Particles.PuffSize:Vector2f=&Multiply_Vector2D_ByFloat(Float=$System.PopScale, Vector2D=<50 50>)
    Particles.VisibilityTag=$Emitter.SpawnGroup
   ) {isSetParameters=true}
   &InitializeParticle(
    Lifetime=&MergeHelper_FloatSwitch(
     0=&RandomRangeFloat(Minimum=0.2, Maximum=0.25)
     1=&RandomRangeFloat(Minimum=0.15, Maximum=0.25)
     2=?
     3=?
     Index=0
     Num=::NewEnumerator0
    )
    Position:Vector3f=$System.BurstPosition
    Position Mode=::NewEnumerator1
    Sprite Size Max=<50 40>
    Sprite Size Min=<40 30>
    Sprite Size Mode=::NewEnumerator2
    UsePositionOffset=false
    Write Material Random=true
   )
   &ShapeLocation(
    MaskSpawnGroup=false
    Override Seed=false
    Sphere Radius=&Multiply_Float(
     A=&MergeHelper_FloatSwitch(0=15, 1=4, 2=?, 3=?, Index=$Emitter.SpawnGroup, Num=::NewEnumerator0)
     B=$System.PopScale
    )
    Sphere Surface Distribution=1
    UseOwnerScale=true
   )
   &AddVelocity(
    Module.Use Falloff Distance=false
    Override Seed=false
    Velocity Mode=::NewEnumerator1
    Velocity Origin:Vector3f=$System.BurstPosition
    Velocity Speed=&Multiply_Float(
     A=&RandomRangeFloat(Minimum=25, Maximum=100)
     B=&MergeHelper_FloatSwitch(0=1, 1=0.7, 2=?, 3=?, Index=&SpawnGroup(), Num=::NewEnumerator0)
    )
   )
  ]
  particleUpdateScript [
   &ParticleState()
   &ScaleSpriteSize(
    Scale Factor=&Lerp_Vector2(A=<1.5 1>, B=<1 2.2>, Alpha=$Particles.NormalizedAge)
    Scale Sprite Size Mode=::NewEnumerator2
   )
   &SubUVAnimation(
    Evaluation Type=::NewEnumerator0
    "Sprite Renderer "=DataInterface() {propertyValues {SpriteRenderer=&self:EndImpact_0.NiagaraSpriteRendererProperties_0}}
    SubUV Animation Mode:ENiagara_SubUVLookupModeV2=::NewEnumerator2
    UseEndFrame=false
    UseStartFrame=false
   )
   &SubUVAnimation(
    Evaluation Type=::NewEnumerator0
    "Sprite Renderer "=DataInterface() {propertyValues {SpriteRenderer=&self:EndImpact_0.NiagaraSpriteRendererProperties_1}}
    SubUV Animation Mode:ENiagara_SubUVLookupModeV2=::NewEnumerator2
    UseEndFrame=false
    UseStartFrame=false
   ) {module=SubUVAnimation001}
   &ScaleColor(
    Scale Alpha=&FloatFromCurve(FloatCurve=DataInterface() {propertyValues {Curve {keys [<at 0.6 1>,<at 1 0>]}}})
    Scale RGB=&Multiply_VectorByFloat(
     Vector=<1 1 1>
     Float=&FloatFromCurve(FloatCurve=DataInterface() {propertyValues {Curve {keys [{arriveTangent=-3.0046, interpMode=RCIM_Cubic, leaveTangent=-3.0046, tangentMode=RCTM_User, time=0, value=3}, <at 1 1>]}}})
    )
    ScaleA=true
    ScaleRGB=true
   )
   &DynamicMaterialParameters(
    Index 0 Param 1=1
    Index 0 Param 2=1
    Module.Index0Float0Write=true
    Module.Index0Float1Write=true
    Module.Index0Float2Write=true
    Module.Index0Float3Write=true
   )
   /SetVariables(Particles.DynamicMaterialParameterPuffs:Vector4f=<0 0 1 1>) {isSetParameters=true}
   &Drag(Drag=&RandomRangeFloat(Minimum=4, Maximum=6), Use Linear Drag=true, Use Rotational Drag=false)
   &ScaleSpriteSize(Scale Sprite Size Mode=::NewEnumerator0, Uniform Scale Factor=$System.PopScale) {module=ScaleSpriteSize001}
   &SolveForcesAndVelocity(Clamp Velocity=false, Limit Acceleration=false)
  ]
  rendererClasses [&NiagaraSpriteRendererProperties]
  renderers [
   {rendererClass=&NiagaraSpriteRendererProperties}
   {rendererClass=&NiagaraSpriteRendererProperties, rendererIndex=1}
  ]
 }
 IntroHeart {
  emitterSpawnScript [&InitializeAudioPlayer(Audio Player=DataInterface() {propertyValues {
      Attenuation=null
      Concurrency=null
      MaxPlaysPerTick=10
      SoundToPlay=/CosmeticCompanions/Assets/Audio_Shared/Interaction/Petting/SC_Mimosa_Interaction_React_Default
      bAllowLoopingOneShotSounds=false
      bLimitPlaysPerTick=true
      bOnlyActiveDuringGameplay=false
      bStopWhenComponentIsDestroyed=true
      ConfigurationUserParameter {parameter {name=None, typeDefHandle {registeredTypeIndex=1}, varData []}}
      ParameterNames []
     }})]
  emitterUpdateScript [
   &EmitterState(
    Life Cycle Mode=::NewEnumerator1
    Loop Behavior=::NewEnumerator1
    Loop Duration=&Multiply_Float(A=$System.IntroLength, B=0.75)
    UseLoopDelay=false
   )
   /Niagara/Modules/Audio/PlayAudio(Play Audio Trigger:NiagaraBool=&SetBoolByFloatComparison(A=$Emitter.Age, B=0.54)) {module=PlayAudio001}
   &SpawnBurst_Instantaneous(Use Spawn Probability=false, UseLoopCountLimit=false)
   /SetVariables(
    Emitter.DistortionStrength:NiagaraFloat=&Lerp_Float(
     A=0.5
     B=0
     Alpha=&FloatFromCurve(
      FloatCurve=DataInterface() {propertyValues {Curve {
         defaultValue=1
         keys [<at 0.3 0 RCIM_Cubic>, {arriveTangent=2.9113, interpMode=RCIM_Cubic, leaveTangent=2.9113, tangentMode=RCTM_User, time=0.7, value=1}]
        }}}
      CurveIndex=&Divide_Float(A=$Emitter.Age, B=$Emitter.CurrentLoopDuration)
     )
    )
    Emitter.Offset:NiagaraFloat=&Lerp_Float(
     A=-10
     B=25.5
     Alpha=&FloatFromCurve(
      FloatCurve=DataInterface() {propertyValues {Curve {
         defaultValue=0
         keys [
          {arriveTangent=1.2674, interpMode=RCIM_Cubic, leaveTangent=1.2674, tangentMode=RCTM_User, time=0, value=0}
          {arriveTangent=1.9203, interpMode=RCIM_Cubic, leaveTangent=1.9203, tangentMode=RCTM_User, time=0.244, value=0.55}
          {arriveTangent=0.5485, interpMode=RCIM_Cubic, leaveTangent=0.5485, tangentMode=RCTM_User, time=0.6, value=0.95}
          {arriveTangent=-0.2501, interpMode=RCIM_Cubic, leaveTangent=-0.2501, tangentMode=RCTM_User, time=1, value=1}
         ]
        }}}
      CurveIndex=&Divide_Float(A=$Emitter.Age, B=$Emitter.CurrentLoopDuration)
     )
    )
    Emitter.ScaleY:NiagaraFloat=&FloatFromCurve(
     FloatCurve=DataInterface() {propertyValues {Curve {keys [<at 0 4>,<at 0.8 1>]}}}
     CurveIndex=&Divide_Float(A=$Emitter.Age, B=$Emitter.CurrentLoopDuration)
    )
    Emitter.Tiling:NiagaraFloat=&Lerp_Float(
     A=0.07
     B=0.1
     Alpha=&FloatFromCurve(
      FloatCurve=DataInterface() {propertyValues {Curve {
         defaultValue=1
         keys [{arriveTangent=2.3592, interpMode=RCIM_Cubic, time=0.1, value=0}, {interpMode=RCIM_Cubic, leaveTangent=2.459, time=1, value=1}]
        }}}
      CurveIndex=&Divide_Float(A=$Emitter.Age, B=$Emitter.CurrentLoopDuration)
     )
    )
   ) {isSetParameters=true}
  ]
  particleSpawnScript [
   /SetVariables(Particles.VisibilityTag=$Emitter.SpawnGroup) {isSetParameters=true}
   &InitializeParticle(
    Lifetime=&Add_Float(A=$System.IntroLength, B=0.13)
    Mesh Scale Mode=::NewEnumerator3
    Mesh Uniform Scale=&Multiply_Float(A=0.2, B=$System.Scale)
    Position:Vector3f=$System.BurstPosition
    Position Mode=::NewEnumerator1
    UsePositionOffset=true
    Write Material Random=true
   )
  ]
  particleUpdateScript [
   &ParticleState()
   &ScaleMeshSize(Scale Factor=&Multiply_VectorByFloat(
     Vector=<1 1 1>
     Float=&FloatFromCurve(FloatCurve=DataInterface() {propertyValues {Curve {keys [<at 0 0.8>,<at 0.7 1.4>,<at 0.9341 1.5>,<at 1 0.6>]}}})
    ))
   &CurlNoiseForce(Bias Noise=false, Noise Frequency=5000, Noise Strength=1.5, Pan Noise=false)
   &SolveForcesAndVelocity(Clamp Velocity=false, Limit Acceleration=false)
   /SetVariables(Particles.Position=&AddVectorToPosition(
     Position=Linked() {linkedVariable {name=System.BurstPosition, type=Vector3f}}
     Vector=&Multiply_VectorByFloat(
      Vector=$Output.CurlNoiseForce.CurlNoiseForce
      Float=&FloatFromCurve(
       FloatCurve=DataInterface() {propertyValues {Curve {keys [
           {arriveTangent=0.7188, interpMode=RCIM_Cubic, leaveTangent=0.7188, tangentMode=RCTM_User, time=0, value=0.2}
           {interpMode=RCIM_Cubic, leaveTangent=0.7145, time=0.5, value=1}
          ]}}}
       CurveIndex=&Subtract_Float(A=$Particles.Age, B=$Emitter.CurrentLoopDuration)
      )
     )
    )) {isSetParameters=true}
   &ScaleColor(
    Scale Alpha=&FloatFromCurve(FloatCurve=DataInterface() {propertyValues {Curve {keys [<at 0 0>,<at 0.5 1>]}}})
    ScaleA=true
    ScaleRGB=false
   )
   &DynamicMaterialParameters(
    Index 0 Param 1=1
    Index 0 Param 2=1
    Index 0 Param 3=1
    Module.Index0Float0Write=true
    Module.Index0Float1Write=true
    Module.Index0Float2Write=true
    Module.Index0Float3Write=true
   )
  ]
  rendererClasses [&NiagaraMeshRendererProperties]
  renderers [{rendererClass=&NiagaraMeshRendererProperties, rendererIndex=0}]
 }
 LW_Skull {rendererClasses [], renderers []}
 LW_SkullShockwave {rendererClasses [], renderers []}
 MorphSpikes {
  emitterUpdateScript [
   &EmitterState(Life Cycle Mode=::NewEnumerator1, Loop Behavior=::NewEnumerator1, Loop Delay=0.2, Loop Duration=0.8, UseLoopDelay=true)
   &SpawnRate(
    SpawnGroup=1
    SpawnRate=&MakeFloatFromInt(Int=&MakeIntFromFloat(Float=&FN_PickFloatDependOnPlatform(TargetPlatformActive=15, TargetPlatformNoActive=10)))
    Use Spawn Probability=false
   ) {module=SpawnRate002}
   &SpawnBurst_Instantaneous(Spawn Group=2, Use Spawn Probability=false, UseLoopCountLimit=false)
   &SpawnBurst_Instantaneous(
    Spawn Group=3
    Spawn Time=&Add_Float(A=$System.IntroLength, B=-0.1)
    Use Spawn Probability=false
    UseLoopCountLimit=false
   ) {module=SpawnBurst_Instantaneous001}
   /SetVariables(Emitter.EmitterRandom:NiagaraFloat=&RandomRangeFloat(Minimum=0, Maximum=360)) {isSetParameters=true}
  ]
  particleSpawnScript [
   /SetVariables(Particles.VisibilityTag=$Emitter.SpawnGroup) {isSetParameters=true}
   &InitializeParticle(
    Color Mode=::NewEnumerator0
    Lifetime=&MergeHelper_FloatSwitch(
     0=0
     1=&UniformRangedFloat(Minimum=0.4, Maximum=0.7)
     2=1
     3=0.3
     Index=$Particles.VisibilityTag
     Num=::NewEnumerator2
    )
    Mesh Scale Mode=::NewEnumerator3
    Mesh Uniform Scale=&Multiply_Float(A=0.3, B=$System.Scale)
    Position Offset=&AddVectorToPosition(Position=Linked() {linkedVariable {name=System.BurstPosition, type=Vector3f}}, Vector=<0 0 0>)
    Position Offset Coordinate Space=::World
    Sprite Rotation Angle=&MergeHelper_FloatSwitch(
     0=&RandomRangeFloat(Minimum=0, Maximum=1)
     1=&Add_Float(
      A=&Multiply_Float(
       A=72
       B=&MakeFloatFromInt(Int=/CascadeToNiagaraConverter/NiagaraScripts/DynamicInputScripts/SequentialInt(Use Array Data:NiagaraBool=false, First Number:NiagaraInt32=0, Sequence:NiagaraDataInterfaceArrayInt32=?, Last Number:NiagaraInt32=5))
      )
      B=$Emitter.EmitterRandom
     )
     2=?
     3=?
     Index=$Particles.VisibilityTag
     Num=::NewEnumerator0
    )
    Sprite Rotation Mode=::NewEnumerator0
    Sprite Size Mode=::NewEnumerator3
    Sprite UV Mode=::NewEnumerator4
    Sprite UV Scale:Vector2f=&MergeHelper_Vector2Switch(
     0=&MakeVector2D(
      X=&UniformAOrBFloat(A=-1, Evaluation Type=::NewEnumerator0, Recalculate Random Each Loop=true, Override Seed=false)
      Y=&UniformAOrBFloat(A=-1, Evaluation Type=::NewEnumerator0, Recalculate Random Each Loop=true, Override Seed=false)
     )
     1=&MakeVector2D(
      X=&UniformAOrBFloat(A=-1, Evaluation Type=::NewEnumerator0, Recalculate Random Each Loop=true, Override Seed=false)
      Y=1
     )
     2=?
     3=?
     4=?
     Index=$Particles.VisibilityTag
     Num=::NewEnumerator0
    )
    Uniform Sprite Size=&Multiply_Float(
     A=&MergeHelper_FloatSwitch(
      0=&RandomRangeFloat(Minimum=25, Maximum=35)
      1=&RandomRangeFloat(Minimum=28, Maximum=35)
      2=?
      3=?
      Index=$Particles.VisibilityTag
      Num=::NewEnumerator0
     )
     B=$System.Scale
    )
    UsePositionOffset=true
    Write Material Random=true
   )
   &DynamicMaterialParameters(
    Index 0 Param 1=&RandomRangeFloat(Minimum=-0.05, Maximum=0.05)
    Index 0 Param 2=&RandomRangeFloat(Minimum=0, Maximum=32)
    Module.Index0Float0Write=true
    Module.Index0Float1Write=true
    Module.Index0Float2Write=true
    Module.Index0Float3Write=true
   )
   &CameraOffset(Camera Offset Amount=&MergeHelper_FloatSwitch(0=5, 1=-5, 2=0, 3=5, Index=$Emitter.SpawnGroup, Num=::NewEnumerator2))
   &SubUVAnimation(
    End Frame Range Override=3
    Evaluation Type=::NewEnumerator0
    Output Alternate Renderer Binding=true
    "Sprite Renderer "=DataInterface() {propertyValues {SpriteRenderer=&self:MorphSpikes_0.NiagaraSpriteRendererProperties_0}}
    SubUV Animation Mode:ENiagara_SubUVLookupModeV2=::NewEnumerator2
    UseEndFrame=true
    UseStartFrame=true
   )
   &InitialMeshOrientation(
    Mesh Orientation Mode=::NewEnumerator7
    Rotation=&MakeVector(X=&RandomRangeFloat(Minimum=0, Maximum=1), Y=0, Z=0)
    Use Rotation Vector=true
   )
  ]
  particleUpdateScript [
   &ParticleState()
   &Drag(Drag=3, Use Linear Drag=true, Use Rotational Drag=true)
   &ScaleColor(
    Scale Alpha=&MergeHelper_FloatSwitch(
     0=&FloatFromCurve(
      FloatCurve=DataInterface() {propertyValues {Curve {keys [
          {interpMode=RCIM_Cubic, leaveTangentWeight=0.0333, tangentMode=RCTM_User, time=0, value=0}
          {arriveTangentWeight=0.0333, interpMode=RCIM_Cubic, leaveTangentWeight=0.3, tangentMode=RCTM_User, time=0.1, value=1}
          {arriveTangentWeight=0.3, interpMode=RCIM_Cubic, tangentMode=RCTM_User, time=1, value=0}
         ]}}}
      Scale Curve=10
     )
     1=&Multiply_Float(
      A=&FloatFromCurve(
       FloatCurve=DataInterface() {propertyValues {Curve {keys [
           {interpMode=RCIM_Cubic, leaveTangentWeight=0.0333, tangentMode=RCTM_User, time=0, value=0}
           {arriveTangentWeight=0.0333, interpMode=RCIM_Cubic, leaveTangentWeight=0.3, tangentMode=RCTM_User, time=0.3, value=1}
           {arriveTangentWeight=0.3, interpMode=RCIM_Cubic, tangentMode=RCTM_User, time=1, value=0}
          ]}}}
       Scale Curve=5
      )
      B=&FloatFromCurve(FloatCurve=DataInterface() {propertyValues {Curve {keys [<at 0.5 1 RCIM_Cubic>,<at 1.5 0 RCIM_Cubic>]}}}, CurveIndex=$Engine.System.Age)
     )
     2=?
     3=?
     Index=$Particles.SpawnGroup
     Num=::NewEnumerator0
    )
    ScaleA=true
    ScaleRGB=true
   )
   &AccelerationForce(Acceleration=&MergeHelper_VectorSwitch(0=<0 0 -10>, 1=<0 0 0>, 2=<0 0 0>, 3=<0 0 0>, Index=$Particles.VisibilityTag, Num=::NewEnumerator2))
   &FN_Sprite_Attributes(
    Color=&MakeLinearColorFromVectorAndFloat(
     "Vector (RGB)"=&MakeVectorFromLinearColorRGB(Color=&ColorFromCurve(ColorCurve=DataInterface() {propertyValues {
         AlphaCurve {keys [<at 0 1 RCIM_Cubic>]}
         BlueCurve {keys [<at 0.1996 0.3437 RCIM_Cubic>,<at 0.5387 0.0026 RCIM_Cubic>]}
         GreenCurve {keys [<at 0.1996 3 RCIM_Cubic>,<at 0.5387 0.7448 RCIM_Cubic>]}
         RedCurve {keys [<at 0.1996 1.7753 RCIM_Cubic>,<at 0.5387 0.2525 RCIM_Cubic>]}
        }}))
     "Float (Alpha)"=&Multiply_Float(
      A=&FloatFromCurve(FloatCurve=DataInterface() {propertyValues {Curve {keys [<at 0 1 RCIM_Cubic>,<at 1 0 RCIM_Cubic>]}}}, Scale Curve=0.3)
      B=&FloatFromCurve(FloatCurve=DataInterface() {propertyValues {Curve {keys [<at 0.5 1 RCIM_Cubic>,<at 1.5 0 RCIM_Cubic>]}}}, CurveIndex=$Engine.System.Age)
     )
    )
    Create Camera Offset Attribute=true
    Create Color Attribute=true
    Create SpriteSize Attribute=true
    Module Output:ENiagara_AttributeSamplingApplyOutput=::NewEnumerator1
    SpriteSize=&Multiply_Vector2D_ByFloat(
     Float=$System.Scale
     Vector2D=&Multiply_Vector2D_ByFloat(
      Float=&FloatFromCurve(FloatCurve=DataInterface() {propertyValues {Curve {keys [
           {arriveTangent=2.0035, interpMode=RCIM_Cubic, leaveTangent=2.0035, tangentMode=RCTM_User, time=0, value=0.6}
           {tangentMode=RCTM_User, time=1, value=1.5}
          ]}}})
      Vector2D=&Multiply_Vector2D(
       A=&RandomRangeVector2D(Minimum=<9 20>, Maximum=<13 25>)
       B=&Vector2DFromCurve(
        Vector2Curve=DataInterface() {propertyValues {
          XCurve {keys [<at 0.5 1 RCIM_Cubic>,<at 1 0.5 RCIM_Cubic>]}
          YCurve {keys [
            {arriveTangent=1.0106, interpMode=RCIM_Cubic, leaveTangent=1.0106, tangentMode=RCTM_User, time=0, value=1}
            {tangentMode=RCTM_User, time=1, value=1.5}
           ]}
         }}
        ScaleCurve=<1 1>
       )
      )
     )
    )
   )
   &FN_Sprite_Attributes(
    Color=&Multiply_LinearColorByFloat(
     Linear Color=<rgba 0.2401 0.7 0 1>
     Float=&FloatFromCurve(
      FloatCurve=DataInterface() {propertyValues {Curve {keys [
          <at 0.075 0 RCIM_Cubic>
          {arriveTangentWeight=0.0896, interpMode=RCIM_Cubic, leaveTangentWeight=0.2868, tangentMode=RCTM_User, time=0.2, value=1}
          {arriveTangentWeight=0.2134, interpMode=RCIM_Cubic, leaveTangentWeight=0.2134, tangentMode=RCTM_User, time=0.6, value=1}
          <at 1 0 RCIM_Cubic>
         ]}}}
      Scale Curve=0.9
     )
    )
    Create Color Attribute=true
    Create Dynamic Material Parameter Attributes=true
    Create SpriteSize Attribute=true
    Index 0 Param 1=3
    Index 0 Param 2=1
    Index 0 Param 3=1
    Index 0 Param 4=1
    Module Output:ENiagara_AttributeSamplingApplyOutput=::NewEnumerator1
    Module.Index0Float0Write=true
    Module.Index0Float1Write=true
    Module.Index0Float2Write=true
    Module.Index0Float3Write=true
    SpriteSize=&Vector2DFromFloat(Value=&Multiply_Float(A=40, B=$System.Scale))
   ) {module=FN_Sprite_Attributes001}
   &FN_Mesh_Attributes(
    Color=&MultiplyLinearColor(
     Linear Color=<rgba 1.552 2 0.2447 1>
     Scale Factor=&MakeVector4FromVectorAndFloat(
      "Vector (RGB)"=<1 1 1>
      "Float (Alpha)"=&FloatFromCurve(FloatCurve=DataInterface() {propertyValues {Curve {keys [<at 0.2 1 RCIM_Cubic>,<at 1 0>]}}})
     )
    )
    Create Color Attribute=true
    Create Dynamic Parameter Attribute=true
    Index 0 Param 2:NiagaraFloat=&FloatFromCurve(FloatCurve=DataInterface() {propertyValues {Curve {keys [<at 0 2>,<at 0.5 1.5>,<at 1 0.8>]}}})
    Index 0 Param 3:NiagaraFloat=&FloatFromCurve(FloatCurve=DataInterface() {propertyValues {Curve {keys [{arriveTangent=5.5668, interpMode=RCIM_Cubic, leaveTangent=5.5668, tangentMode=RCTM_User, time=0, value=1.2}]}}})
    Index 0 Param 4:NiagaraFloat=&RandomRangeFloat(Minimum=0, Maximum=1)
    Module Output=::NewEnumerator1
    Module.Index0Float0Write=true
    Module.Index0Float1Write=true
    Module.Index0Float2Write=true
    Module.Index0Float3Write=true
   )
   &SolveForcesAndVelocity(Clamp Velocity=false, Limit Acceleration=false)
  ]
  rendererClasses [&NiagaraSpriteRendererProperties, &NiagaraMeshRendererProperties]
  renderers [
   {rendererClass=&NiagaraSpriteRendererProperties}
   {rendererClass=&NiagaraSpriteRendererProperties, rendererIndex=1}
   {rendererClass=&NiagaraMeshRendererProperties, rendererIndex=2}
  ]
 }
}
summary {
 systemName=NS_ReactFX_NibbleMoss
 emitters [
  {emitterName=LW_Skull, rendererClasses []}
  {emitterName=DeathBurst, rendererClasses [&NiagaraSpriteRendererProperties, &NiagaraMeshRendererProperties]}
  {emitterName=EndImpact, rendererClasses [&NiagaraSpriteRendererProperties]}
  {emitterName=MorphSpikes, rendererClasses [&NiagaraSpriteRendererProperties, &NiagaraMeshRendererProperties]}
  {emitterName=IntroHeart, rendererClasses [&NiagaraMeshRendererProperties]}
  {emitterName=LW_SkullShockwave, rendererClasses []}
 ]
 userVariables [
  User.bIsFrontEnd {type=NiagaraBool}
  User.bIsFrontEndPreview {type=NiagaraBool}
 ]
}
