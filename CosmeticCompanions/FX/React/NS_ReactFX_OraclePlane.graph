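# NS_ReactFX_OraclePlane is a Niagara cosmetic companion reaction effect: a pulsing, glowing pink heart with a petting reaction sound.
# NE_Heart lives for System.IntroLength, with curve-driven sizes, material parameters and time-of-day-scaled glow alpha.
# NE_Confetti_In swirls motes toward System.BurstPosition and NE_Star01 adds stars there, before NE_Confetti001 bursts outward just ahead of intro end.
# Spawn groups feed VisibilityTag and per-group switches, letting each emitter carry several particle types.
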NiagaraSystem /CosmeticCompanions/FX/React/NS_ReactFX_OraclePlane
dependencies {
 usedDataInterfaces [&NiagaraDataInterfaceAudioPlayer, &NiagaraDataInterfaceCurve]
 usedDynamicInputs [&SimulationPosition, /Niagara/DynamicInputs/TypeConversions/MakeCustomPositionFromBool]
 usedModules [
  &self:SystemSpawnScript.SystemScriptSource.SystemScriptGraph.NiagaraNodeAssignment_0.SetVariables_#1
  &self:SystemSpawnScript.SystemScriptSource.SystemScriptGraph.NiagaraNodeAssignment_4.SetVariables_#2
  &self:SystemSpawnScript.SystemScriptSource.SystemScriptGraph.NiagaraNodeAssignment_1.SetVariables_#3
  &self:SystemSpawnScript.SystemScriptSource.SystemScriptGraph.NiagaraNodeAssignment_6.SetVariables_#4
  &self:SystemSpawnScript.SystemScriptSource.SystemScriptGraph.NiagaraNodeAssignment_7.SetVariables_#5
  &self:SystemSpawnScript.SystemScriptSource.SystemScriptGraph.NiagaraNodeAssignment_2.SetVariables_#6
  &self:SystemSpawnScript.SystemScriptSource.SystemScriptGraph.NiagaraNodeAssignment_3.SetVariables_#7
  &self:SystemSpawnScript.SystemScriptSource.SystemScriptGraph.NiagaraNodeAssignment_5.SetVariables_#8
  &self:SystemSpawnScript.SystemScriptSource.SystemScriptGraph.NiagaraNodeAssignment_9.SetVariables_#9
  &self:SystemSpawnScript.SystemScriptSource.SystemScriptGraph.NiagaraNodeAssignment_10.SetVariables_#10
  &self:SystemSpawnScript.SystemScriptSource.SystemScriptGraph.NiagaraNodeAssignment_11.SetVariables_#11
  &self:SystemSpawnScript.SystemScriptSource.SystemScriptGraph.NiagaraNodeAssignment_12.SetVariables_#12
  &self:SystemSpawnScript.SystemScriptSource.SystemScriptGraph.NiagaraNodeAssignment_13.SetVariables_#13
  &self:SystemSpawnScript.SystemScriptSource.SystemScriptGraph.NiagaraNodeAssignment_14.SetVariables_#14
  &self:SystemSpawnScript.SystemScriptSource.SystemScriptGraph.NiagaraNodeAssignment_15.SetVariables_#15
  &SystemState
 ]
}
emitters {
 LWE_AirGlow {rendererClasses [], renderers []}
 LWE_Corona {rendererClasses [], renderers []}
 LWE_Rays {rendererClasses [], renderers []}
 NE_Confetti001 {
  emitterUpdateScript [
   &EmitterState()
   &SpawnBurst_Instantaneous(Spawn Count=30, Spawn Time=&Subtract_Float(A=$System.IntroLength, B=0.05), Use Spawn Probability=false) {module=SpawnBurst_Instantaneous002}
   &SpawnBurst_Instantaneous(
    Spawn Count=10
    Spawn Group=1
    Spawn Time=&Subtract_Float(A=$System.IntroLength, B=0.05)
    Use Spawn Probability=false
   ) {module=SpawnBurst_Instantaneous001}
   &SpawnBurst_Instantaneous(
    Spawn Count=35
    Spawn Group=2
    Spawn Time=&Subtract_Float(A=$System.IntroLength, B=0.05)
    Use Spawn Probability=false
   )
  ]
  particleSpawnScript [
   /SetVariables(Particles.VisibilityTag=&SpawnGroup()) {isSetParameters=true}
   &InitializeParticle(
    Lifetime Max=&MergeHelper_FloatSwitch(0=2, 1=2, 2=2.5, 3=?, Index=&SpawnGroup(), Num=::NewEnumerator1)
    Lifetime Min:NiagaraFloat=&MergeHelper_FloatSwitch(0=0.3, 1=0.5, 2=0.2, 3=?, Index=&SpawnGroup(), Num=::NewEnumerator1)
    Lifetime Mode=::NewEnumerator1
    Position Offset:NiagaraPosition=$System.BurstPosition
    Position Offset Coordinate Space=::Local
    Sprite Rotation Mode=::NewEnumerator2
    Sprite Size Mode=::NewEnumerator1
    Uniform Sprite Size Max=6
    Uniform Sprite Size Min=3
    UsePositionOffset=true
    Write Material Random=true
   ) {module=InitializeParticle001}
   /SetVariables(Particles.SpriteRotation:NiagaraFloat=&MergeHelper_FloatSwitch(
     0=0
     1=&RandomRangeFloat(Minimum=0, Maximum=360)
     2=&RandomRangeFloat(Minimum=0, Maximum=360)
     3=?
     Index=&SpawnGroup()
     Num=::NewEnumerator1
    )) {isSetParameters=true}
   &ShapeLocation(
    MaskSpawnGroup=false
    Override Seed=false
    Sphere Radius=&MergeHelper_FloatSwitch(0=10, 1=2, 2=0, 3=?, Index=&SpawnGroup(), Num=::NewEnumerator1)
    UseOwnerScale=true
   )
   &AddVelocity(
    Module.Use Falloff Distance=false
    Origin Offset:NiagaraPosition=$System.BurstPosition
    Override Seed=false
    Velocity Mode=::NewEnumerator1
    Velocity Speed=&MergeHelper_FloatSwitch(
     0=&RandomRangeFloat(Minimum=50, Maximum=300)
     1=&RandomRangeFloat(Minimum=200, Maximum=250)
     2=&RandomRangeFloat(Minimum=150, Maximum=250)
     3=?
     Index=&SpawnGroup()
     Num=::NewEnumerator1
    )
   ) {module=AddVelocity002}
   &AddVelocity(Velocity=&MergeHelper_VectorSwitch(
     0=&RandomRangeVector(Minimum=<0 0 50>, Maximum=<0 0 100>)
     1=&RandomRangeVector(Minimum=<0 0 50>, Maximum=<0 0 100>)
     2=&RandomRangeVector(Minimum=<0 0 50>, Maximum=<0 0 100>)
     3=?
     Index=&SpawnGroup()
     Num=::NewEnumerator1
    )) {module=AddVelocity003}
   /SetVariables(Particles.SmallStarSpriteSize:Vector2f=&Multiply_Vector2D_ByFloat(Float=&RandomRangeFloat(Minimum=8, Maximum=15), Vector2D=<1 1>)) {isSetParameters=true}
   /SetVariables(Particles.BigStarSpriteSize:Vector2f=&Multiply_Vector2D_ByFloat(Float=&RandomRangeFloat(Minimum=8, Maximum=15), Vector2D=<1 1>)) {isSetParameters=true}
   /SetVariables(Particles.ConfettiSpriteSize:Vector2f=&Multiply_Vector2D_ByFloat(Float=&RandomRangeFloat(Minimum=3, Maximum=6), Vector2D=<1 1>)) {isSetParameters=true}
   /SetVariables(Particles.SmallStarColor:LinearColor=<rgba 1 0.4766 0.642 5>) {isSetParameters=true}
   /SetVariables(Particles.BigStarColor:LinearColor=<rgba 1.2891 0.3808 0.7339 5>) {isSetParameters=true}
   /SetVariables() {isSetParameters=true}
  ]
  particleUpdateScript [
   &ParticleState()
   &SpriteRotationRate(Rotation Rate=&MergeHelper_FloatSwitch(
     0=0
     1=&RandomRangeFloat(Minimum=-100, Maximum=100)
     2=0
     3=?
     Index=&SpawnGroup()
     Num=::NewEnumerator1
    ))
   &Drag(
    Drag=&MergeHelper_FloatSwitch(
     0=&RandomRangeFloat(Minimum=8, Maximum=10)
     1=&RandomRangeFloat(Minimum=5, Maximum=8)
     2=&RandomRangeFloat(Minimum=5, Maximum=10)
     3=?
     Index=&SpawnGroup()
     Num=::NewEnumerator1
    )
    Use Linear Drag=true
    Use Rotational Drag=false
   ) {module=Drag001}
   &GravityForce(Gravity=&MergeHelper_VectorSwitch(
     0=&RandomRangeVector(Minimum=<0 0 -300>, Maximum=<0 0 -200>)
     1=&RandomRangeVector(Minimum=<0 0 -200>, Maximum=<1 1 -200>)
     2=&RandomRangeVector(Minimum=<0 0 -360>, Maximum=<1 1 -200>)
     3=?
     Index=&SpawnGroup()
     Num=::NewEnumerator1
    )) {module=GravityForce001}
   /SetVariables(Particles.UpdateSmallStarSize:Vector2f=&Multiply_Vector2D_ByFloat(
     Float=&FloatFromCurve(FloatCurve=DataInterface() {propertyValues {Curve {keys [<at -0.0022 0.051>,<at 0.1043 0.9697>,<at 0.2753 0.2958>,<at 0.4326 0.9033>,<at 0.6128 0.1144>,<at 0.7818 0.632>,<at 1.0011 0.002>]}}})
     Vector2D=$Particles.Initial.SmallStarSpriteSize
    )) {isSetParameters=true}
   /SetVariables(Particles.UpdateBigStarSize:Vector2f=&Multiply_Vector2D_ByFloat(
     Float=&FloatFromCurve(FloatCurve=DataInterface() {propertyValues {Curve {defaultValue=1, keys [{tangentMode=RCTM_User, time=0, value=0}, <at 0.163 1>, {tangentMode=RCTM_User, time=1, value=0}]}}})
     Vector2D=$Particles.Initial.BigStarSpriteSize
    )) {isSetParameters=true}
   /SetVariables(Particles.UpdateConfettiSize:Vector2f=&Multiply_Vector2D_ByFloat(
     Float=&FloatFromCurve(FloatCurve=DataInterface() {propertyValues {Curve {keys [
          {arriveTangent=-0.4596, interpMode=RCIM_Cubic, leaveTangent=-0.4596, tangentMode=RCTM_User, time=0.55, value=1}
          {arriveTangent=-4.9031, interpMode=RCIM_Cubic, leaveTangent=-4.9031, tangentMode=RCTM_User, time=1, value=0}
         ]}}})
     Vector2D=$Particles.Initial.ConfettiSpriteSize
    )) {isSetParameters=true}
   /SetVariables(Particles.UpdateConfettiColor:LinearColor=&MultiplyLinearColor(
     Linear Color=$Particles.Initial.ConfettiColor
     Scale Factor=&MakeVector4FromVectorAndFloat(
      "Vector (RGB)"=&MakeVector(
       X=1
       Y=&Add_Float(
        A=&RandomRangeFloat(Minimum=0, Maximum=1000)
        B=&Lerp_Float(A=0, B=&RandomRangeFloat(Minimum=25, Maximum=150), Alpha=$Particles.NormalizedAge)
       )
       Z=1
      )
      "Float (Alpha)"=1
     )
    )) {isSetParameters=true}
   /SetVariables(Particles.UpdateSmallStarColor:LinearColor=&MultiplyLinearColor(
     Linear Color=$Particles.Initial.SmallStarColor
     Scale Factor=&MakeVector4_FromLinearColor(LinearColor=&MakeLinearColorFromVectorAndFloat(
       "Vector (RGB)"=<1 1 1>
       "Float (Alpha)"=&FloatFromCurve(FloatCurve=DataInterface() {propertyValues {Curve {keys [<at -0.0092 0.0144>,<at 0.0967 0.9767>,<at 1.0011 0.002>]}}})
      ))
    )) {isSetParameters=true}
   /SetVariables(Particles.UpdateBigStarColor:LinearColor=&MultiplyLinearColor(
     Linear Color=$Particles.Initial.SmallStarColor
     Scale Factor=&MakeVector4_FromLinearColor(LinearColor=&MakeLinearColorFromVectorAndFloat(
       "Vector (RGB)"=<1 1 1>
       "Float (Alpha)"=&FloatFromCurve(FloatCurve=DataInterface() {propertyValues {Curve {keys [<at -0.0092 0.0144>,<at 0.0967 0.9767>,<at 1.0011 0.002>]}}})
      ))
    )) {isSetParameters=true}
   &ScaleSpriteSize(Scale Factor=&Vector2DFromFloat(Value=&FloatFromCurve(FloatCurve=DataInterface() {propertyValues {Curve {keys [
          {arriveTangent=-0.4596, interpMode=RCIM_Cubic, leaveTangent=-0.4596, tangentMode=RCTM_User, time=0.55, value=1}
          {arriveTangent=-4.9031, interpMode=RCIM_Cubic, leaveTangent=-4.9031, tangentMode=RCTM_User, time=1, value=0}
         ]}}}))) {enabled=false}
   &ScaleColor(
    Scale RGB=&MakeVector(
     X=1
     Y=&Add_Float(
      A=&RandomRangeFloat(Minimum=0, Maximum=1000)
      B=&Lerp_Float(A=0, B=&RandomRangeFloat(Minimum=25, Maximum=150), Alpha=$Particles.NormalizedAge)
     )
     Z=1
    )
    ScaleA=false
    ScaleRGB=true
   ) {enabled=false}
   &SolveForcesAndVelocity(Clamp Velocity=false, Limit Acceleration=false)
  ]
  rendererClasses [&NiagaraSpriteRendererProperties]
  renderers [
   {rendererClass=&NiagaraSpriteRendererProperties}
   {rendererClass=&NiagaraSpriteRendererProperties, rendererIndex=1}
   {rendererClass=&NiagaraSpriteRendererProperties, rendererIndex=2}
  ]
 }
 NE_Confetti_In {
  emitterUpdateScript [
   &EmitterState(Life Cycle Mode=::NewEnumerator1, Loop Behavior=::NewEnumerator1, Loop Duration=0.5, UseLoopDelay=false)
   &SpawnRate(SpawnRate=100, Use Spawn Probability=false)
   &SpawnRate(SpawnGroup=1, SpawnRate=50, Use Spawn Probability=false) {module=SpawnRate001}
  ]
  particleSpawnScript [
   /SetVariables(Particles.VisibilityTag=&SpawnGroup()) {isSetParameters=true}
   &InitializeParticle(
    Lifetime Max=0.4
    Lifetime Min=0.2
    Lifetime Mode=::NewEnumerator1
    Position Offset=&Subtract_Position(A=$System.BurstPosition, B=&FN_FrontEndVector3Modifier(Gameplay=<0 0 0>, Front End=<0 0 0>))
    Position Offset Coordinate Space=::Local
    Sprite Rotation Mode=::NewEnumerator2
    Sprite Size Mode=::NewEnumerator1
    Uniform Sprite Size Max=15
    UsePositionOffset=true
    Write Material Random=true
   ) {module=InitializeParticle001}
   &ShapeLocation(MaskSpawnGroup=false, Override Seed=false, Sphere Radius=80, UseOwnerScale=true)
   /SetVariables(Particles.Motes_color:LinearColor=<rgba 5 2.1109 3.6044 1>) {isSetParameters=true}
  ]
  particleUpdateScript [
   &ParticleState()
   &PointAttractionForce(
    Attraction Radius=500
    Attraction Strength=500
    Attractor Position Offset:Vector3f=&Subtract_Position(A=$System.BurstPosition, B=&FN_FrontEndVector3Modifier(Gameplay=<0 0 0>, Front End=<0 0 0>))
    Attractor Position Offset Coordinate Space:ENiagaraCoordinateSpace=::Local
    Kill Within Radius=false
    Module.Use Falloff=true
    UseAttractorOffset=true
   )
   &Drag(Drag=&RandomRangeFloat(Minimum=5, Maximum=20), Use Linear Drag=true, Use Rotational Drag=false)
   &VortexForce(
    Origin Pull Amount=1500
    UseOriginOffset=true
    Vortex Axis=<0 1 0>
    Vortex Axis Coordinate Space:ENiagaraCoordinateSpace=::Local
    Vortex Force Amount=3000
    Vortex Origin Offset:Vector3f=&Subtract_Position(A=$System.BurstPosition, B=&FN_FrontEndVector3Modifier(Gameplay=<0 0 0>, Front End=<0 0 0>))
    Vortex Origin Offset Coordinate Space:ENiagaraCoordinateSpace=::Local
   )
   &ScaleSpriteSize(Uniform Curve Sprite Scale=DataInterface() {propertyValues {Curve {
       defaultValue=0
       keys [
        {interpMode=RCIM_Cubic, tangentMode=RCTM_User, time=0, value=1}
        {arriveTangent=-2.3313, interpMode=RCIM_Cubic, leaveTangent=-2.3313, tangentMode=RCTM_User, time=1, value=0}
       ]
      }}}) {module=ScaleSpriteSize001}
   &SolveForcesAndVelocity(Clamp Velocity=false, Limit Acceleration=false)
   &ScaleSpriteSize(Scale Factor=&Vector2DFromFloat(Value=&FloatFromCurve(FloatCurve=DataInterface() {propertyValues {Curve {keys [
          {arriveTangent=-0.4596, interpMode=RCIM_Cubic, leaveTangent=-0.4596, tangentMode=RCTM_User, time=0.55, value=1}
          {arriveTangent=-2.2017, interpMode=RCIM_Cubic, leaveTangent=-2.2017, tangentMode=RCTM_User, time=1, value=0.5}
         ]}}})))
   &ScaleColor(
    Scale RGB=&MakeVector(
     X=1
     Y=&Add_Float(
      A=&RandomRangeFloat(Minimum=0, Maximum=1000)
      B=&Lerp_Float(A=0, B=&RandomRangeFloat(Minimum=25, Maximum=150), Alpha=$Particles.NormalizedAge)
     )
     Z=1
    )
    ScaleA=false
    ScaleRGB=true
   )
   /SetVariables(Particles.UpdateMotesColor:LinearColor=&Multiply_LinearColorByFloat(
     Linear Color=$Particles.Initial.Motes_color
     Float=&FloatFromCurve(FloatCurve=DataInterface() {propertyValues {Curve {keys [<at -0.0092 0.0144>,<at 0.0967 0.9767>,<at 1.0011 0.002>]}}})
    )) {isSetParameters=true}
  ]
  rendererClasses [&NiagaraSpriteRendererProperties]
  renderers [
   {rendererClass=&NiagaraSpriteRendererProperties}
   {rendererClass=&NiagaraSpriteRendererProperties, rendererIndex=1}
  ]
 }
 NE_Heart {
  emitterSpawnScript [
   &InitializeAudioPlayer(Audio Player=DataInterface() {propertyValues {
      Attenuation=null
      Concurrency=null
      MaxPlaysPerTick=10
      SoundToPlay=/CosmeticCompanions/Assets/Audio_Shared/Interaction/Petting/SC_Mimosa_Interaction_React_Default
      bAllowLoopingOneShotSounds=false
      bLimitPlaysPerTick=true
      bOnlyActiveDuringGameplay=false
      bStopWhenComponentIsDestroyed=true
      ConfigurationUserParameter {parameter {name=None, typeDefHandle {registeredTypeIndex=1}, varData []}}
      ParameterNames []
     }})
   /Niagara/Modules/Audio/PlayAudio() {module=PlayAudio001}
  ]
  emitterUpdateScript [
   &EmitterState()
   &SpawnBurst_Instantaneous(Use Spawn Probability=false, UseLoopCountLimit=false)
   &SpawnBurst_Instantaneous(Spawn Group=1, Use Spawn Probability=false, UseLoopCountLimit=false) {module=SpawnBurst_Instantaneous001}
  ]
  particleSpawnScript [
   /SetVariables(Particles.VisibilityTag=&SpawnGroup()) {isSetParameters=true}
   &InitializeParticle(
    Lifetime=$System.IntroLength
    Sprite Rotation Mode=::NewEnumerator0
    Sprite Size=&MergeHelper_Vector2Switch(0=<11 10>, 1=<16 16>, 2=?, 3=?, 4=?, Index=$Particles.VisibilityTag, Num=::NewEnumerator0)
    Sprite Size Mode=::NewEnumerator4
    UsePositionOffset=false
    Write Material Random=true
   )
   /SetVariables(Particles.Glow Color=<rgba 1 0.0452 0.7029 0.35>) {isSetParameters=true}
   /SetVariables(Particles.Glow2SpriteSize:Vector2f=&Multiply_Vector2D_ByFloat(Float=17, Vector2D=<1 1>)) {isSetParameters=true}
   /SetVariables(Particles.Glow3SpriteSize:Vector2f=&Multiply_Vector2D_ByFloat(Float=15.0653, Vector2D=<1 1>)) {isSetParameters=true}
   /SetVariables(Particles.Glow4SpriteSize:Vector2f=&Multiply_Vector2D_ByFloat(Float=14, Vector2D=<1 1>)) {isSetParameters=true}
   /SetVariables(Particles.Glow2Color:LinearColor=<rgba 0.5365 0.0053 0.3648 1>) {isSetParameters=true}
   /SetVariables(Particles.Glow3Color:LinearColor=<rgba 1 0.0098 0.68 1>) {isSetParameters=true}
   /SetVariables(Particles.Glow4Color:LinearColor=<rgba 1 0.8582 0.3384 1>) {isSetParameters=true}
  ]
  particleUpdateScript [
   &ParticleState()
   &ScaleSpriteSize(
    Scale Sprite Size Mode=::NewEnumerator0
    Uniform Scale Factor=&Add_Float(
     A=0.5
     B=&FloatFromCurve(FloatCurve=DataInterface() {propertyValues {Curve {
         defaultValue=0
         keys [
          {arriveTangent=0.8193, time=0, value=0}
          <at 0.3 2>
          <at 0.5 1.5>
          <at 0.7 2>
          {arriveTangent=-0.9053, interpMode=RCIM_Cubic, leaveTangent=-0.9053, tangentMode=RCTM_User, time=0.8783, value=3}
          {arriveTangent=-0.1202, interpMode=RCIM_Cubic, leaveTangent=-0.1202, tangentMode=RCTM_User, time=1, value=1}
         ]
        }}})
    )
   ) {enabled=false, module=ScaleSpriteSize002}
   /SetVariables(Particles.UpdateHeartSpriteSize:Vector2f=&Multiply_Vector2D_ByFloat(
     Float=&Add_Float(
      A=0.5
      B=&FloatFromCurve(FloatCurve=DataInterface() {propertyValues {Curve {
          defaultValue=0
          keys [
           {arriveTangent=0.8193, time=0, value=0}
           <at 0.3 2>
           <at 0.5 1.5>
           {arriveTangent=-0.9053, interpMode=RCIM_Cubic, leaveTangent=-0.9053, tangentMode=RCTM_User, time=0.8, value=3}
           {arriveTangentWeight=0.3399, leaveTangentWeight=0.3302, time=0.9, value=3}
           {arriveTangent=-0.1202, leaveTangent=-0.1202, time=1, value=1}
          ]
         }}})
     )
     Vector2D=$Particles.Initial.SpriteSize
    )) {isSetParameters=true}
   /SetVariables(Particles.UpdateGlow2SpriteSize:Vector2f=&Multiply_Vector2D_ByFloat(
     Float=&Add_Float(
      A=0.5
      B=&FloatFromCurve(FloatCurve=DataInterface() {propertyValues {Curve {
          defaultValue=0
          keys [
           {arriveTangent=0.8193, time=0, value=0}
           <at 0.3 2>
           <at 0.5009 1.0235>
           {arriveTangent=-0.9053, interpMode=RCIM_Cubic, leaveTangent=-0.9053, tangentMode=RCTM_User, time=0.8, value=3}
           {arriveTangentWeight=0.3399, leaveTangentWeight=0.3302, time=0.9, value=3}
           {arriveTangent=-0.1202, leaveTangent=-0.1202, time=1, value=1}
          ]
         }}})
     )
     Vector2D=$Particles.Initial.Glow2SpriteSize
    )) {isSetParameters=true}
   /SetVariables(Particles.UpdateGlow3SpriteSize:Vector2f=&Multiply_Vector2D_ByFloat(
     Float=&Add_Float(
      A=0.5
      B=&FloatFromCurve(FloatCurve=DataInterface() {propertyValues {Curve {
          defaultValue=0
          keys [
           {arriveTangent=0.8193, time=0, value=0}
           <at 0.3 2>
           <at 0.5056 -0.3882>
           {arriveTangent=-0.9053, interpMode=RCIM_Cubic, leaveTangent=-0.9053, tangentMode=RCTM_User, time=0.8, value=3}
           {arriveTangentWeight=0.3399, leaveTangentWeight=0.3302, time=0.9, value=3}
           {arriveTangent=-0.1202, leaveTangent=-0.1202, time=1, value=1}
          ]
         }}})
     )
     Vector2D=$Particles.Initial.Glow3SpriteSize
    )) {isSetParameters=true}
   /SetVariables(Particles.UpdateGlow4SpriteSize:Vector2f=&Multiply_Vector2D_ByFloat(
     Float=&Add_Float(
      A=0.184
      B=&FloatFromCurve(FloatCurve=DataInterface() {propertyValues {Curve {
          defaultValue=0
          keys [
           {arriveTangent=0.8193, time=0, value=0}
           <at 0.3 2>
           <at 0.5 1.5>
           {arriveTangent=-0.9053, interpMode=RCIM_Cubic, leaveTangent=-0.9053, tangentMode=RCTM_User, time=0.8, value=3}
           {arriveTangentWeight=0.3399, leaveTangentWeight=0.3302, time=0.9, value=3}
           {arriveTangent=-0.1202, leaveTangent=-0.1202, time=1, value=1}
          ]
         }}})
     )
     Vector2D=$Particles.Initial.Glow4SpriteSize
    )) {isSetParameters=true}
   /SetVariables(Particles.UpdateGlowColor:LinearColor=&MakeLinearColorFromVectorAndFloat(
     "Vector (RGB)"=&MakeVectorFromLinearColorRGB(Color=$Particles.Initial.Glow Color)
     "Float (Alpha)"=&Multiply_Float(
      A=&FloatFromCurve(
       FloatCurve=DataInterface() {propertyValues {Curve {keys [
           {arriveTangent=11.9433, time=0, value=0}
           {arriveTangent=2, interpMode=RCIM_Cubic, leaveTangent=2, tangentMode=RCTM_User, time=0.5, value=0}
           {arriveTangent=5, interpMode=RCIM_Cubic, leaveTangent=5, time=0.5693, value=3}
           <at 0.7 1 RCIM_Cubic>
          ]}}}
       Scale Curve=0.5
      )
      B=&Multiply_Float(A=1, B=&FNTimeOfDayScalar(Night Scale=0.5, Morning Scale=0.25, Evening Scale=0.25))
     )
    )) {isSetParameters=true}
   /SetVariables(Particles.UpdateGlow2Color:LinearColor=&MakeLinearColorFromVectorAndFloat(
     "Vector (RGB)"=&MakeVectorFromLinearColorRGB(Color=$Particles.Initial.Glow2Color)
     "Float (Alpha)"=&Multiply_Float(
      A=&FloatFromCurve(FloatCurve=DataInterface() {propertyValues {Curve {keys [
           {arriveTangent=11.9433, time=0, value=0}
           {arriveTangent=-0.0808, interpMode=RCIM_Cubic, leaveTangent=-0.0808, tangentMode=RCTM_User, time=0.43, value=0}
           {arriveTangent=8.5873, interpMode=RCIM_Cubic, leaveTangent=8.5873, tangentMode=RCTM_User, time=0.65, value=1.584}
           {arriveTangent=-0.1193, interpMode=RCIM_Cubic, leaveTangent=-0.1193, tangentMode=RCTM_User, time=0.7, value=0.7417}
          ]}}})
      B=&Multiply_Float(A=1, B=&FNTimeOfDayScalar(Night Scale=0.5, Morning Scale=0.25, Evening Scale=0.25))
     )
    )) {isSetParameters=true}
   /SetVariables(Particles.UpdateGlow3Color:LinearColor=&MakeLinearColorFromVectorAndFloat(
     "Vector (RGB)"=&MakeVectorFromLinearColorRGB(Color=$Particles.Initial.Glow3Color)
     "Float (Alpha)"=&Multiply_Float(
      A=&FloatFromCurve(
       FloatCurve=DataInterface() {propertyValues {Curve {keys [
           {arriveTangent=11.9433, time=0, value=0}
           {arriveTangent=2, interpMode=RCIM_Cubic, leaveTangent=2, tangentMode=RCTM_User, time=0.337, value=0.319}
           {arriveTangent=1.0958, interpMode=RCIM_Cubic, leaveTangent=1.0958, time=0.5937, value=1.584}
           <at 0.7 0.7168 RCIM_Cubic>
          ]}}}
       Scale Curve=0.8
      )
      B=&Multiply_Float(A=1, B=&FNTimeOfDayScalar(Night Scale=0.5, Morning Scale=0.25, Evening Scale=0.25))
     )
    )) {isSetParameters=true}
   /SetVariables(Particles.UpdateGlow4Color:LinearColor=&MakeLinearColorFromVectorAndFloat(
     "Vector (RGB)"=&MakeVectorFromLinearColorRGB(Color=$Particles.Initial.Glow4Color)
     "Float (Alpha)"=&Multiply_Float(
      A=&FloatFromCurve(
       FloatCurve=DataInterface() {propertyValues {Curve {keys [
           {arriveTangent=11.9433, time=0, value=0}
           {arriveTangent=2, interpMode=RCIM_Cubic, leaveTangent=2, tangentMode=RCTM_User, time=0.3393, value=0}
           <at 0.4398 1.3131 RCIM_Cubic>
           <at 0.6 0 RCIM_Cubic>
          ]}}}
       Scale Curve=0.2
      )
      B=&Multiply_Float(A=1, B=&FNTimeOfDayScalar(Night Scale=0.5, Morning Scale=0.25, Evening Scale=0.25))
     )
    )) {isSetParameters=true}
   /SetVariables(Particles.DynamicMaterialParameter01:Vector4f=<1 1 1 0.8>) {isSetParameters=true}
   &ScaleColor(
    Scale Alpha=&FloatFromCurve(FloatCurve=DataInterface() {propertyValues {Curve {
        defaultValue=0
        keys [
         {arriveTangent=9.0725, interpMode=RCIM_Cubic, leaveTangent=9.0725, tangentMode=RCTM_User, time=0, value=0}
         {arriveTangent=-0.2441, leaveTangent=-0.2441, time=0.1408, value=0.9847}
         {arriveTangent=-1.146, leaveTangent=-1.146, time=0.95, value=1}
         {arriveTangent=-0.0405, interpMode=RCIM_Cubic, leaveTangent=-0.0405, tangentMode=RCTM_User, time=1, value=0}
        ]
       }}})
    ScaleA=true
    ScaleRGB=true
   )
   &OffsetPosition(Position Offset=&Lerp_Vector(
     A=$System.HeartIntroSpawnPosition
     B=Linked() {linkedVariable {name=System.BurstPosition, type=NiagaraPosition}}
     Alpha=&FloatFromCurve(FloatCurve=DataInterface() {propertyValues {Curve {keys [{arriveTangent=0.7913, time=0, value=1}, <at 0.3 1>, {arriveTangent=-0.1074, tangentMode=RCTM_User, time=1, value=1}]}}})
    ))
   &DynamicMaterialParameters(
    Index 0 Param 1=&FloatFromCurve(FloatCurve=DataInterface() {propertyValues {Curve {keys [
         <at 0 -1>
         {interpMode=RCIM_Cubic, tangentMode=RCTM_User, time=0.4, value=-1}
         {arriveTangent=19.8658, interpMode=RCIM_Cubic, leaveTangent=19.8658, tangentMode=RCTM_User, time=0.6, value=1}
        ]}}})
    Index 0 Param 2=1
    Index 0 Param 3=&FloatFromCurve(FloatCurve=DataInterface() {propertyValues {Curve {
        defaultValue=1
        keys [
         {tangentMode=RCTM_User, time=0, value=-0.5}
         {interpMode=RCIM_Cubic, tangentMode=RCTM_User, time=0.2, value=-0.5}
         {arriveTangent=15.0323, interpMode=RCIM_Cubic, leaveTangent=15.0323, tangentMode=RCTM_User, time=0.4, value=1}
        ]
       }}})
    Index 0 Param 4=&FloatFromCurve(FloatCurve=DataInterface() {propertyValues {Curve {keys [
         <at 0 -1>
         {interpMode=RCIM_Cubic, tangentMode=RCTM_User, time=0.4, value=-1}
         {arriveTangent=19.8658, interpMode=RCIM_Cubic, leaveTangent=19.8658, tangentMode=RCTM_User, time=0.6, value=1}
        ]}}})
    Module.Index0Float0Write=true
    Module.Index0Float1Write=true
    Module.Index0Float2Write=true
    Module.Index0Float3Write=true
   )
   &SpriteRotationRate(Rotation Rate=&MergeHelper_FloatSwitch(0=0, 1=90, 2=?, 3=?, Index=$Particles.VisibilityTag, Num=::NewEnumerator0))
  ]
  rendererClasses [&NiagaraSpriteRendererProperties]
  renderers [
   {rendererClass=&NiagaraSpriteRendererProperties}
   {rendererClass=&NiagaraSpriteRendererProperties, rendererIndex=1}
   {rendererClass=&NiagaraSpriteRendererProperties, rendererIndex=2}
   {rendererClass=&NiagaraSpriteRendererProperties, rendererIndex=3}
   {rendererClass=&NiagaraSpriteRendererProperties, rendererIndex=4}
   {rendererClass=&NiagaraSpriteRendererProperties, rendererIndex=5}
  ]
 }
 NE_Star01 {
  emitterUpdateScript [
   &EmitterState(Life Cycle Mode=::NewEnumerator1, Loop Behavior=::NewEnumerator1, Loop Duration=0.65, UseLoopDelay=false)
   &SpawnRate(
    SpawnRate=&FloatFromCurve(
     FloatCurve=DataInterface() {propertyValues {Curve {keys [<at 0 0>,<at 0.25 0>,<at 0.3 1>,<at 1 1>]}}}
     CurveIndex=$Emitter.NormalizedLoopAge
     Scale Curve=60
    )
    Use Spawn Probability=false
   ) {module=SpawnRate001}
   &SpawnRate(
    SpawnGroup=1
    SpawnRate=&FloatFromCurve(
     FloatCurve=DataInterface() {propertyValues {Curve {keys [<at 0 0>,<at 0 0>,<at 0.4408 0>,<at 0.5 0.9941>,<at 0.6 1>,<at 0.8 0>]}}}
     CurveIndex=$Emitter.NormalizedLoopAge
     Scale Curve=60
    )
    Use Spawn Probability=false
   )
  ]
  particleSpawnScript [
   &InitializeParticle(
    Color=<rgba 1 0.4766 0.642 5>
    Lifetime Max=&MergeHelper_FloatSwitch(0=0.5, 1=0.7, 2=?, 3=?, Index=&SpawnGroup(), Num=::NewEnumerator0)
    Lifetime Min:NiagaraFloat=&MergeHelper_FloatSwitch(0=0.3, 1=0.4, 2=?, 3=?, Index=&SpawnGroup(), Num=::NewEnumerator0)
    Lifetime Mode=::NewEnumerator1
    Position Offset:NiagaraPosition=$System.BurstPosition
    Position Offset Coordinate Space=::Local
    Sprite Rotation Mode=::NewEnumerator2
    Sprite Size Mode=::NewEnumerator1
    UsePositionOffset=true
    Write Material Random=true
   ) {module=InitializeParticle001}
   &ShapeLocation(
    MaskSpawnGroup=false
    Override Seed=false
    Sphere Radius=&MergeHelper_FloatSwitch(0=17, 1=23, 2=?, 3=?, Index=&SpawnGroup(), Num=::NewEnumerator0)
    Sphere Surface Distribution:NiagaraFloat=&MergeHelper_FloatSwitch(0=0.5, 1=1, 2=?, 3=?, Index=&SpawnGroup(), Num=::NewEnumerator0)
    UseOwnerScale=true
   )
   /SetVariables(Particles.VisibilityTag=&SpawnGroup()) {isSetParameters=true}
   /SetVariables(Particles.SpriteRotation:NiagaraFloat=&MergeHelper_FloatSwitch(
     0=0
     1=&RandomRangeFloat(Minimum=0, Maximum=360)
     2=?
     3=?
     Index=&SpawnGroup()
     Num=::NewEnumerator0
    )) {isSetParameters=true}
   /SetVariables(Particles.SmallStarSpriteSize:Vector2f=&Multiply_Vector2D_ByFloat(Float=&RandomRangeFloat(Minimum=5, Maximum=9), Vector2D=<1 1>)) {isSetParameters=true}
  ]
  particleUpdateScript [
   &ParticleState()
   &Drag(Drag=&RandomRangeFloat(Minimum=8, Maximum=10), Use Linear Drag=true, Use Rotational Drag=false)
   &ScaleColor(
    Scale Alpha=&FloatFromCurve(FloatCurve=DataInterface() {propertyValues {Curve {keys [<at -0.0092 0.0144>,<at 0.1 1>,<at 1 0.3>]}}})
    ScaleA=true
    ScaleRGB=true
   )
   /SetVariables(Particles.UpdateBigStarSize:Vector2f=&Multiply_Vector2D_ByFloat(
     Float=&FloatFromCurve(
      FloatCurve=DataInterface() {propertyValues {Curve {
         defaultValue=0
         keys [
          {arriveTangent=0.8193, time=0, value=0}
          <at 0.3 2>
          <at 0.5 1.5>
          {arriveTangent=-0.9053, interpMode=RCIM_Cubic, leaveTangent=-0.9053, tangentMode=RCTM_User, time=0.7706, value=3}
          {arriveTangentWeight=0.3399, leaveTangentWeight=0.3302, time=0.9206, value=3}
          {arriveTangent=-0.1202, leaveTangent=-0.1202, time=1, value=1}
         ]
        }}}
      Scale Curve=0.384
     )
     Vector2D=$Particles.Initial.SpriteSize
    )) {isSetParameters=true}
   &ScaleSpriteSize(Scale Factor=&Vector2DFromFloat(Value=&FloatFromCurve(FloatCurve=DataInterface() {propertyValues {Curve {keys [<at -0.0022 0.051>,<at 0.1 1>,<at 1 0>]}}}, Scale Curve=0.384))) {enabled=false}
   /SetVariables(Particles.UpdateSmallStarSize:Vector2f=&Multiply_Vector2D_ByFloat(
     Float=&FloatFromCurve(FloatCurve=DataInterface() {propertyValues {Curve {keys [<at 0 1>,<at 1 0>]}}})
     Vector2D=$Particles.Initial.SmallStarSpriteSize
    )) {isSetParameters=true}
   &SolveForcesAndVelocity(Clamp Velocity=false, Limit Acceleration=false)
  ]
  rendererClasses [&NiagaraSpriteRendererProperties]
  renderers [
   {rendererClass=&NiagaraSpriteRendererProperties}
   {rendererClass=&NiagaraSpriteRendererProperties, rendererIndex=1}
  ]
 }
}
summary {
 systemName=NS_ReactFX_OraclePlane
 emitters [
  {emitterName=NE_Heart, rendererClasses [&NiagaraSpriteRendererProperties]}
  {emitterName=LWE_AirGlow, rendererClasses []}
  {emitterName=LWE_Rays, rendererClasses []}
  {emitterName=LWE_Corona, rendererClasses []}
  {emitterName=NE_Confetti_In, rendererClasses [&NiagaraSpriteRendererProperties]}
  {emitterName=NE_Star01, rendererClasses [&NiagaraSpriteRendererProperties]}
  {emitterName=NE_Confetti001, rendererClasses [&NiagaraSpriteRendererProperties]}
 ]
 userVariables [
  User.bIsFrontEnd {type=NiagaraBool}
  User.bIsFrontEndPreview {type=NiagaraBool}
 ]
}
