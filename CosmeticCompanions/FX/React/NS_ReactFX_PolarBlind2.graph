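# Niagara reaction effect for a cosmetic companion that shows a growing heart with a petting-reaction sound, an incoming arrow, a fiery skull and a dark burst.
# Timed emitters share System.BurstPosition: heart, screen-space arrow, flipbook skull, then a three-group burst with smoke and streak attributes.
# A delayed, descending flame follows, feeding curl-noise embers through a particle attribute reader.
# Front-end modifiers switch arrow placement and rotation between menu and gameplay.
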NiagaraSystem /CosmeticCompanions/FX/React/NS_ReactFX_PolarBlind2
dependencies {
 usedDataInterfaces [
  &NiagaraDataInterfaceAudioPlayer
  &NiagaraDataInterfaceSpriteRendererInfo
  &NiagaraDataInterfaceCurve
  &NiagaraDataInterfaceColorCurve
  &NiagaraDataInterfaceArrayColor
  &NiagaraDataInterfaceVectorCurve
  &NiagaraDataInterfaceParticleRead
 ]
 usedDynamicInputs [/Niagara/DynamicInputs/TypeConversions/MakeCustomPositionFromBool]
 usedModules [
  &self:SystemSpawnScript.SystemScriptSource.SystemScriptGraph.NiagaraNodeAssignment_0.SetVariables_#1
  &self:SystemSpawnScript.SystemScriptSource.SystemScriptGraph.NiagaraNodeAssignment_1.SetVariables_#2
  &self:SystemSpawnScript.SystemScriptSource.SystemScriptGraph.NiagaraNodeAssignment_7.SetVariables_#3
  &self:SystemSpawnScript.SystemScriptSource.SystemScriptGraph.NiagaraNodeAssignment_4.SetVariables_#4
  &self:SystemSpawnScript.SystemScriptSource.SystemScriptGraph.NiagaraNodeAssignment_8.SetVariables_#5
  &SystemState
 ]
}
emitters {
 Burst_Embers {
  emitterUpdateScript [
   &EmitterState()
   &SpawnParticlesFromOtherEmitter(
    Attribute Reader=DataInterface() {propertyValues {EmitterBinding {bindingMode=Other, emitterName=Burst_Flame}}}
    Spawn Rate=&FloatFromCurve(
     FloatCurve=DataInterface() {propertyValues {Curve {keys [
         {interpMode=RCIM_Cubic, tangentMode=RCTM_User, time=0, value=1}
         {interpMode=RCIM_Cubic, leaveTangentWeight=0.0988, tangentMode=RCTM_User, time=0, value=1}
         {arriveTangentWeight=0.0931, interpMode=RCIM_Cubic, leaveTangentWeight=0.2328, tangentMode=RCTM_User, time=0.2, value=1}
         {arriveTangent=-6.8841, interpMode=RCIM_Cubic, leaveTangent=-6.8841, tangentMode=RCTM_User, time=0.7, value=0.0064}
        ]}}}
     CurveIndex=$Emitter.NormalizedLoopAge
     Scale Curve=30
    )
    Spawn Rate Per Particle Cap=150
    Use Maximum Age=false
    Use Minimum Age=false
   )
  ]
  particleSpawnScript [
   &InitializeParticle(
    Color=<rgba 0 0 0 1>
    Lifetime Max=0.5
    Lifetime Min=0.35
    Lifetime Mode=::NewEnumerator1
    Sprite Size Mode=::NewEnumerator3
    Uniform Sprite Size=&MakeFloatFromVector2d(Vector 2D=&Lerp_Vector2(
      A=<1 2>
      B=<1.5 3>
      Alpha=&FloatFromCurve(
       FloatCurve=DataInterface() {propertyValues {Curve {keys [
           {interpMode=RCIM_Cubic, tangentMode=RCTM_User, time=0, value=0}
           {interpMode=RCIM_Cubic, tangentMode=RCTM_User, time=0.3, value=0}
           {interpMode=RCIM_Cubic, tangentMode=RCTM_User, time=0.8, value=1}
          ]}}}
       CurveIndex=&RandomRangeFloat(Minimum=0.5, Maximum=1)
      )
     ))
    UsePositionOffset=false
    Write Material Random=true
   )
   &AddVelocity(Velocity=&MakeVector(
     X=&RandomRangeFloat(Minimum=-75, Maximum=75)
     Y=&RandomRangeFloat(Minimum=-75, Maximum=75)
     Z=&RandomRangeFloat(Minimum=25, Maximum=40)
    )) {module=AddVelocity001}
   &SampleParticlesFromOtherEmitter(
    Applied Position Offset:Vector3f=<0 0 3.3671>
    Emitter Sampling Mode:ENiagara_EmitterLocSamplingMode=::NewEnumerator0
    Transform Data Option=1
   )
  ]
  particleUpdateScript [
   &ParticleState()
   &ScaleSpriteSize(Uniform Curve Sprite Scale=DataInterface() {propertyValues {Curve {keys [
        {arriveTangent=0.0565, interpMode=RCIM_Cubic, leaveTangent=0.0565, tangentMode=RCTM_User, time=0, value=1}
        {arriveTangent=-0.025, interpMode=RCIM_Cubic, leaveTangent=-0.025, tangentMode=RCTM_User, time=1, value=0}
       ]}}}) {module=ScaleSpriteSize001}
   &Drag(Drag=2, Use Linear Drag=true, Use Rotational Drag=false)
   &ScaleColor(
    Scale Alpha=&FloatFromCurve(FloatCurve=DataInterface() {propertyValues {Curve {keys [
         {arriveTangent=0.3132, interpMode=RCIM_Cubic, leaveTangent=0.3132, tangentMode=RCTM_User, time=0, value=1}
         {arriveTangent=-3.1355, interpMode=RCIM_Cubic, leaveTangent=-3.1355, tangentMode=RCTM_User, time=1, value=0}
        ]}}})
    Scale RGB=&Multiply_VectorByFloat(
     Vector=<1 1 1>
     Float=&FloatFromCurve(
      FloatCurve=DataInterface() {propertyValues {Curve {keys [
          {arriveTangent=-10.4526, interpMode=RCIM_Cubic, leaveTangent=-10.4526, tangentMode=RCTM_User, time=0, value=5}
          {interpMode=RCIM_Cubic, tangentMode=RCTM_User, time=1, value=1}
         ]}}}
      Scale Curve=2
     )
    )
    ScaleA=true
    ScaleRGB=true
   )
   &CurlNoiseForce(
    Curl Noise Cone Mask Angle=90
    Curl Noise Cone Mask Falloff Angle=35
    Mask Curl Noise=true
    Noise Frequency=15
    Noise Strength=200
    Pan Noise=true
    Pan Noise Field=<0.25 0.25 0.25>
   )
   &SolveForcesAndVelocity(Clamp Velocity=false, Limit Acceleration=false)
  ]
  rendererClasses [&NiagaraSpriteRendererProperties]
  renderers [{rendererClass=&NiagaraSpriteRendererProperties}]
 }
 Burst_Flame {
  emitterUpdateScript [
   &EmitterState(
    Inactive Response=::NewEnumerator1
    Life Cycle Mode=::NewEnumerator1
    Loop Behavior=::NewEnumerator1
    Loop Delay=0.956
    Loop Duration=3.4
    UseLoopDelay=true
   )
   &SpawnBurst_Instantaneous(Use Spawn Probability=false, UseLoopCountLimit=false) {module=SpawnBurst_Instantaneous001}
  ]
  particleSpawnScript [
   &InitializeParticle(
    Color Mode=::NewEnumerator0
    Lifetime=0.6
    Position:Vector3f=$System.BurstPosition
    Position Mode=::NewEnumerator1
    Sprite Size Mode=::NewEnumerator3
    Uniform Sprite Size=3.5482
    UsePositionOffset=false
    Write Material Random=true
   )
   &AddVelocity(Velocity=<0 0 -90>)
  ]
  particleUpdateScript [
   &ParticleState()
   &ScaleVelocity(Velocity Scale=&VectorFromFloat(Value=&FloatFromCurve(FloatCurve=DataInterface() {propertyValues {Curve {keys [<at 0 1 RCIM_Cubic>, {arriveTangentWeight=0.7301, interpMode=RCIM_Cubic, tangentMode=RCTM_User, time=1, value=0.35}]}}})))
   &SolveForcesAndVelocity(Clamp Velocity=false, Limit Acceleration=false)
   /SetVariables(
    Particles.Flame Color:LinearColor=&ColorFromCurve(ColorCurve=DataInterface() {propertyValues {
       AlphaCurve {keys [<at 0.5 1 RCIM_Cubic>,<at 1 0 RCIM_Cubic>]}
       BlueCurve {keys [<at 0 0 RCIM_Cubic>]}
       GreenCurve {keys [<at 0 0 RCIM_Cubic>]}
       RedCurve {keys [<at 0 0 RCIM_Cubic>]}
      }})
    Particles.Flame Pivot Offset:Vector2f=&FN_FrontEndVector2Modifier(Front End:Vector2f=<0.2159 0.8725>, Gameplay:Vector2f=<0.2159 0.8725>)
    Particles.Flame Sprite Size:Vector2f=&Vector2DFromFloat(Value=13)
   ) {isSetParameters=true}
   /SetVariables(Particles.PivotOffset:Vector2f=&FN_FrontEndVector2Modifier(Front End:Vector2f=<-0.5746 0.8025>, Gameplay:Vector2f=<-0.5746 0.8025>)) {isSetParameters=true}
   &ScaleColor(
    Linear Color Curve=DataInterface() {propertyValues {
      AlphaCurve {keys [<at 0.5006 1>,<at 0.9701 0>]}
      BlueCurve {keys [<at 0 3>,<at 0.8527 1>]}
      GreenCurve {keys [<at 0 3>,<at 0.8527 1>]}
      RedCurve {keys [<at 0 3>,<at 0.8527 1>]}
     }}
    Scale Mode=2
   )
  ]
  rendererClasses [&NiagaraSpriteRendererProperties]
  renderers [
   {rendererClass=&NiagaraSpriteRendererProperties}
   {rendererClass=&NiagaraSpriteRendererProperties, rendererIndex=1}
  ]
 }
 Intro_Arrow {
  emitterUpdateScript [
   &EmitterState()
   &SpawnBurst_Instantaneous(Spawn Time=0.35, Use Spawn Probability=false, UseLoopCountLimit=false)
  ]
  particleSpawnScript [&InitializeParticle(
    Color Mode=::NewEnumerator0
    Lifetime=0.4
    Position:Vector3f=$System.BurstPosition
    Position Mode=::NewEnumerator1
    Position Offset=<168.4807 0 22.5601>
    Position Offset Coordinate Space=::Local
    Sprite Rotation Angle=&FN_FrontEndFloatModifier(Gameplay=25.6399, Front End=40)
    Sprite Rotation Mode=::NewEnumerator0
    Sprite Size Mode=::NewEnumerator3
    Uniform Sprite Size=20
    UsePositionOffset=true
    Write Material Random=true
   )]
  particleUpdateScript [
   &ParticleState()
   /BRCosmetics/Animation/Game/MainPlayer/Emotes/QuarterClaspJoin/FX/InitScreenSpaceSprite()
   /SetVariables(Particles.Position=&Lerp_Vector(
     A=Linked() {linkedVariable {name=Particles.Initial.Position, type=NiagaraPosition}}
     B=&FN_FrontEndVector3Modifier(Gameplay=<50 0 0>, Front End=<25 0 -20>)
     Alpha=&FloatFromCurve(FloatCurve=DataInterface() {propertyValues {Curve {keys [<at 0 0>,<at 0.3 1>]}}})
    )) {isSetParameters=true}
   &ScaleColor(
    Scale Alpha=&FloatFromCurve(FloatCurve=DataInterface() {propertyValues {Curve {keys [<at 0 1>,<at 0.9 1>,<at 1 0>]}}})
    ScaleA=true
    ScaleRGB=true
   )
   &SubUVAnimation(
    End Frame Range Override=0
    "Sprite Renderer "=DataInterface() {propertyValues {SpriteRenderer=&self:Intro_Arrow_0.NiagaraSpriteRendererProperties_0}}
    Start Frame Offset=2
    UseEndFrame=true
    UseStartFrame=true
   )
   &SolveForcesAndVelocity(Clamp Velocity=false, Limit Acceleration=false)
   /BRCosmetics/Animation/Game/MainPlayer/Emotes/QuarterClaspJoin/FX/ScreenSpaceSprite(Multiplier=0.02)
  ]
  rendererClasses [&NiagaraSpriteRendererProperties]
  renderers [{rendererClass=&NiagaraSpriteRendererProperties}]
 }
 Intro_Burst {
  emitterUpdateScript [
   &EmitterState(Life Cycle Mode=::NewEnumerator1, Loop Behavior=::NewEnumerator1, Loop Delay=0.45, Loop Duration=0.35, UseLoopDelay=true)
   &SpawnBurst_Instantaneous(Spawn Count=35, Use Spawn Probability=false, UseLoopCountLimit=false)
   &SpawnBurst_Instantaneous(Spawn Count=5, Spawn Group=1, Use Spawn Probability=false, UseLoopCountLimit=false) {module=SpawnBurst_Instantaneous001}
   &SpawnBurst_Instantaneous(Spawn Count=2, Spawn Group=2, Use Spawn Probability=false, UseLoopCountLimit=false) {module=SpawnBurst_Instantaneous002}
   &SpawnRate(SpawnRate=20, Use Spawn Probability=false)
   &SpawnRate(SpawnGroup=1, SpawnRate=5, Use Spawn Probability=false) {module=SpawnRate001}
   &SpawnRate(SpawnGroup=2, SpawnRate=8, Use Spawn Probability=false) {module=SpawnRate002}
  ]
  particleSpawnScript [
   &InitializeParticle(
    Color=&SelectLinearColorFromArray(
     Array Sampling Mode=::NewEnumerator0
     Direct Array Index=?
     Color Selection Array=DataInterface() {propertyValues {ColorData [<rgba 0 0 0 1>, <rgba 1 1 1 0.1>]}}
     Random Seed=0
    )
    Lifetime=&MergeHelper_FloatSwitch(
     0=&RandomRangeFloat(Minimum=0.5, Maximum=0.7)
     1=&RandomRangeFloat(Minimum=0.5, Maximum=0.7)
     2=&RandomRangeFloat(Minimum=0.25, Maximum=0.35)
     3=?
     Index=&SpawnGroup()
     Num=::NewEnumerator1
    )
    Position=&MergeHelper_VectorSwitch(
     0=$System.BurstPosition
     1=&AddVectorToPosition(Position=Linked() {linkedVariable {name=System.BurstPosition, type=Vector3f}}, Vector=<0 0 4.3982>)
     2=$System.BurstPosition
     3=?
     Index=&SpawnGroup()
     Num=::NewEnumerator1
    )
    Position Mode=::NewEnumerator1
    Sprite Rotation Mode=::NewEnumerator2
    Sprite Size=&Lerp_Vector2(
     A=<1.5 3>
     B=<1.5 4>
     Alpha=&FloatFromCurve(
      FloatCurve=DataInterface() {propertyValues {Curve {keys [
          {interpMode=RCIM_Cubic, tangentMode=RCTM_User, time=0, value=0}
          {interpMode=RCIM_Cubic, tangentMode=RCTM_User, time=0.3, value=0}
          {interpMode=RCIM_Cubic, tangentMode=RCTM_User, time=0.8, value=1}
         ]}}}
      CurveIndex=&RandomRangeFloat(Minimum=0, Maximum=1)
     )
    )
    Sprite Size Mode=::NewEnumerator4
    UsePositionOffset=false
    Write Material Random=true
   ) {module=InitializeParticle001}
   &AddVelocity(Velocity=&MergeHelper_VectorSwitch(
     0=&MakeVector(
      X=&RandomRangeFloat(Minimum=-75, Maximum=75)
      Y=&RandomRangeFloat(Minimum=-75, Maximum=75)
      Z=&RandomRangeFloat(Minimum=35, Maximum=75)
     )
     1=&RandomRangeVector(Minimum=<0 0 1>, Maximum=<0 0 3>)
     2=&RandomRangeVector(Minimum=<0 0 -55>, Maximum=<0 0 -75>)
     3=?
     Index=&SpawnGroup()
     Num=::NewEnumerator1
    )) {module=AddVelocity001}
   &ShapeLocation(
    MaskSpawnGroup=false
    Override Seed=false
    Sphere Radius=&MergeHelper_FloatSwitch(0=10, 1=10, 2=7, 3=?, Index=&SpawnGroup(), Num=::NewEnumerator1)
    UseOwnerScale=true
   )
   /SetVariables(Particles.VisibilityTag=&SpawnGroup()) {isSetParameters=true}
   /SetVariables() {isSetParameters=true}
   /SetVariables(
    Particles.Smoke Color:LinearColor=<rgba 0 0 0 1>
    Particles.Smoke Sprite Size:Vector2f=&Vector2DFromFloat(Value=30)
   ) {isSetParameters=true}
   /SetVariables(
    Particles.Streak Color:LinearColor=<rgba 0 0 0 1>
    Particles.Streak Sprite Size:Vector2f=&RandomRangeVector2D(Minimum=<2 6>, Maximum=<3 20>)
   ) {isSetParameters=true}
  ]
  particleUpdateScript [
   &ParticleState()
   &SubUVAnimation(
    End Frame Range Override=30
    "Sprite Renderer "=DataInterface() {propertyValues {SpriteRenderer=&self:Intro_Burst_0.NiagaraSpriteRendererProperties_0}}
    Start Frame Range Override=&RandomRangeInt(Minimum=0, Maximum=5)
    UseEndFrame=true
    UseStartFrame=true
   )
   &ScaleColor(
    Scale Alpha=&FloatFromCurve(FloatCurve=DataInterface() {propertyValues {Curve {keys [
         {arriveTangent=0.3132, interpMode=RCIM_Cubic, leaveTangent=0.3132, tangentMode=RCTM_User, time=0, value=1}
         {arriveTangent=-3.1355, interpMode=RCIM_Cubic, leaveTangent=-3.1355, tangentMode=RCTM_User, time=1, value=0}
        ]}}})
    Scale RGB=&Multiply_VectorByFloat(
     Vector=<1 1 1>
     Float=&FloatFromCurve(
      FloatCurve=DataInterface() {propertyValues {Curve {keys [
          {arriveTangent=-10.4526, interpMode=RCIM_Cubic, leaveTangent=-10.4526, tangentMode=RCTM_User, time=0, value=5}
          {interpMode=RCIM_Cubic, tangentMode=RCTM_User, time=1, value=1}
         ]}}}
      Scale Curve=2
     )
    )
    ScaleA=true
    ScaleRGB=true
   )
   &ScaleSpriteSize(Uniform Curve Sprite Scale=DataInterface() {propertyValues {Curve {keys [
        {arriveTangent=0.0565, interpMode=RCIM_Cubic, leaveTangent=0.0565, tangentMode=RCTM_User, time=0, value=1}
        {arriveTangent=-0.025, interpMode=RCIM_Cubic, leaveTangent=-0.025, tangentMode=RCTM_User, time=1, value=0}
       ]}}}) {module=ScaleSpriteSize001}
   &Drag(Drag=1.5, Use Linear Drag=true, Use Rotational Drag=false)
   &CurlNoiseForce(
    Curl Noise Cone Mask Angle:NiagaraFloat=&RandomRangeFloat(Minimum=-90, Maximum=90)
    Curl Noise Cone Mask Falloff Angle=35
    Mask Curl Noise=true
    Noise Frequency=15
    Noise Strength=&MergeHelper_FloatSwitch(0=200, 1=50, 2=0, 3=?, Index=&SpawnGroup(), Num=::NewEnumerator1)
    Pan Noise=true
    Pan Noise Field=<0.25 0.25 0.25>
   )
   &GravityForce(Gravity=&MergeHelper_VectorSwitch(
     0=&VectorFromCurve(
      VectorCurve=DataInterface() {propertyValues {
        XCurve {keys [<at 0 0>,<at 1 0>]}
        YCurve {keys [<at 0 0>,<at 1 0>]}
        ZCurve {keys [<at 0.362 0 RCIM_Cubic>, {arriveTangent=2.8269, interpMode=RCIM_Cubic, leaveTangent=2.8269, tangentMode=RCTM_User, time=1, value=1}]}
       }}
      Scale Curve=<1 1 500>
     )
     1=<0 0 0>
     2=<0 0 0>
     3=?
     Index=&SpawnGroup()
     Num=::NewEnumerator1
    ))
   &SolveForcesAndVelocity(Clamp Velocity=false, Limit Acceleration=false)
   /SetVariables(
    Particles.Smoke Color:LinearColor=&MultiplyLinearColor(
     Linear Color=$Particles.Initial.Smoke Color
     Scale Factor=&Vector4FromCurve(Vector4Curve=DataInterface() {propertyValues {
        AlphaCurve {keys [<at 0 0>,<at 0.4441 1>,<at 1 0>]}
        BlueCurve {keys [<at 0 2>,<at 1 1>]}
        GreenCurve {keys [<at 0 2>,<at 1 1>]}
        RedCurve {keys [<at 0 2>,<at 1 1>]}
       }})
    )
    Particles.Smoke Sprite Size:Vector2f=&Multiply_Vector2D(
     A=$Particles.Initial.Smoke Sprite Size
     B=&Vector2DFromFloat(Value=&FloatFromCurve(FloatCurve=DataInterface() {propertyValues {Curve {keys [<at 0 0.8>,<at 1 1.5>]}}}))
    )
   ) {isSetParameters=true}
   /SetVariables(
    Particles.Streak Color:LinearColor=&MultiplyLinearColor(
     Linear Color=$Particles.Initial.Streak Color
     Scale Factor=&Vector4FromCurve(Vector4Curve=DataInterface() {propertyValues {
        AlphaCurve {keys [<at 0 0>,<at 0.0914 2>,<at 1 0>]}
        BlueCurve {keys [<at 0 10>,<at 1 1>]}
        GreenCurve {keys [<at 0 10>,<at 1 1>]}
        RedCurve {keys [<at 0 10>,<at 1 1>]}
       }})
    )
    Particles.Streak Sprite Size:Vector2f=&Multiply_Vector2D(A=$Particles.Initial.Streak Sprite Size, B=<1 1>)
   ) {isSetParameters=true}
  ]
  rendererClasses [&NiagaraSpriteRendererProperties]
  renderers [
   {rendererClass=&NiagaraSpriteRendererProperties}
   {rendererClass=&NiagaraSpriteRendererProperties, rendererIndex=1}
   {rendererClass=&NiagaraSpriteRendererProperties, rendererIndex=2}
  ]
 }
 Intro_FierySkull {
  emitterUpdateScript [
   &EmitterState(Life Cycle Mode=::NewEnumerator1, Loop Behavior=::NewEnumerator1, Loop Delay=0.46, Loop Duration=2, UseLoopDelay=true)
   &SpawnBurst_Instantaneous(Spawn Group=1, Use Spawn Probability=false, UseLoopCountLimit=false) {module=SpawnBurst_Instantaneous001}
   &SpawnBurst_Instantaneous(Use Spawn Probability=false, UseLoopCountLimit=false)
  ]
  particleSpawnScript [
   &InitializeParticle(
    Lifetime=0.5
    Position:Vector3f=$System.BurstPosition
    Position Mode=::NewEnumerator1
    Sprite Size Mode=::NewEnumerator3
    Uniform Sprite Size=35
    UsePositionOffset=true
    Write Material Random=true
   )
   /SetVariables(Particles.VisibilityTag=&SpawnGroup()) {isSetParameters=true}
  ]
  particleUpdateScript [
   &ParticleState()
   &SubUVAnimation(
    Direct Frame Index:NiagaraFloat=&Lerp_Float(
     A=1
     B=8
     Alpha=&FloatFromCurve(FloatCurve=DataInterface() {propertyValues {Curve {keys [{leaveTangentWeight=0.5751, time=0, value=0}, {arriveTangentWeight=0.3351, time=1, value=1}]}}})
    )
    "Sprite Renderer "=DataInterface() {propertyValues {SpriteRenderer=&self:Intro_FierySkull_0.NiagaraSpriteRendererProperties_0}}
    SubUV Animation Mode:ENiagara_SubUVLookupModeV2=::NewEnumerator5
   )
   /SetVariables(
    Particles.Linear Color_Halo:LinearColor=&ColorFromCurve(ColorCurve=DataInterface() {propertyValues {
       AlphaCurve {keys [<at 0.9445 0.05 RCIM_Cubic>,<at 1 0 RCIM_Cubic>]}
       BlueCurve {keys [<at 0 1>]}
       GreenCurve {keys [<at 0 1>]}
       RedCurve {keys [<at 0 1>]}
      }})
    Particles.SpriteSize_Halo:Vector2f=&Vector2DFromFloat(Value=37.864)
   ) {isSetParameters=true}
   /SetVariables(
    Particles.LinearColor_Electric:LinearColor=&Multiply_LinearColorByFloat(
     Linear Color=&ColorFromCurve(ColorCurve=DataInterface() {propertyValues {
        AlphaCurve {keys [<at 0.9116 1 RCIM_Cubic>]}
        BlueCurve {keys [<at 0.0586 1 RCIM_Constant>,<at 0.1939 1 RCIM_Constant>,<at 0.3325 1 RCIM_Constant>,<at 0.493 1 RCIM_Constant>]}
        GreenCurve {keys [<at 0.0586 1 RCIM_Constant>,<at 0.1939 1 RCIM_Constant>,<at 0.3325 1 RCIM_Constant>,<at 0.493 1 RCIM_Constant>]}
        RedCurve {keys [<at 0.0586 1 RCIM_Constant>,<at 0.1939 1 RCIM_Constant>,<at 0.3325 1 RCIM_Constant>,<at 0.493 1 RCIM_Constant>]}
       }})
     Float=1
    )
    Particles.SpriteSize_Electric:Vector2f=&Vector2DFromFloat(Value=&FloatFromCurve(
      FloatCurve=DataInterface() {propertyValues {Curve {keys [
          {arriveTangent=12.8286, interpMode=RCIM_Cubic, leaveTangent=12.8286, tangentMode=RCTM_User, time=0, value=1}
          <at 0.2 1 RCIM_Cubic>
          <at 1 1 RCIM_Cubic>
         ]}}}
      Scale Curve=25
     ))
   ) {isSetParameters=true}
   /SetVariables(Particles.Color_Shatter:LinearColor=&ColorFromCurve(
     ColorCurve=DataInterface() {propertyValues {
       AlphaCurve {keys [<at 0.3053 1>,<at 0.5076 0 RCIM_Cubic>]}
       BlueCurve {keys [<at 0.5115 0 RCIM_Cubic>]}
       GreenCurve {keys [<at 0.5115 0 RCIM_Cubic>]}
       RedCurve {keys [<at 0.5115 0 RCIM_Cubic>]}
      }}
     CurveIndex=&Multiply_Float(A=$Particles.NormalizedAge, B=1)
    )) {isSetParameters=true}
   &DynamicMaterialParameters(Write Parameter Index 0=::NewEnumerator0)
  ]
  rendererClasses [&NiagaraSpriteRendererProperties]
  renderers [
   {rendererClass=&NiagaraSpriteRendererProperties}
   {rendererClass=&NiagaraSpriteRendererProperties, rendererIndex=1}
  ]
 }
 Intro_Heart {
  emitterSpawnScript [
   &InitializeAudioPlayer(Audio Player=DataInterface() {propertyValues {
      Attenuation=null
      Concurrency=null
      MaxPlaysPerTick=10
      SoundToPlay=/CosmeticCompanions/Assets/Audio_Shared/Interaction/Petting/SC_Mimosa_Interaction_React_Default
      bAllowLoopingOneShotSounds=false
      bLimitPlaysPerTick=true
      bOnlyActiveDuringGameplay=false
      bStopWhenComponentIsDestroyed=true
      ConfigurationUserParameter {parameter {name=None, typeDefHandle {registeredTypeIndex=1}, varData []}}
      ParameterNames []
     }})
   /Niagara/Modules/Audio/PlayAudio() {module=PlayAudio001}
  ]
  emitterUpdateScript [
   &EmitterState(Life Cycle Mode=::NewEnumerator1, Loop Behavior=::NewEnumerator1, Loop Duration=1, UseLoopDelay=false)
   &SpawnBurst_Instantaneous(Use Spawn Probability=false, UseLoopCountLimit=false)
  ]
  particleSpawnScript [
   &InitializeParticle(
    Lifetime=$System.IntroLength
    Position:Vector3f=$System.BurstPosition
    Position Mode=::NewEnumerator1
    Sprite Rotation Mode=::NewEnumerator0
    Sprite Size Mode=::NewEnumerator3
    Uniform Sprite Size=12
    UsePositionOffset=false
    Write Material Random=true
   )
   &ApplyInitialForces()
   &SubUVAnimation(
    Direct Frame Index=8
    "Sprite Renderer "=DataInterface() {propertyValues {SpriteRenderer=&self:Intro_Heart_0.NiagaraSpriteRendererProperties_0}}
    SubUV Animation Mode:ENiagara_SubUVLookupModeV2=::NewEnumerator5
   )
  ]
  particleUpdateScript [
   &ParticleState()
   &ScaleSpriteSize(
    Scale Sprite Size Mode=::NewEnumerator0
    Uniform Scale Factor=&Add_Float(
     A=0.5
     B=&FloatFromCurve(FloatCurve=DataInterface() {propertyValues {Curve {
         defaultValue=0
         keys [
          {arriveTangent=2.0692, interpMode=RCIM_Cubic, leaveTangent=2.0692, tangentMode=RCTM_User, time=0, value=0}
          {arriveTangent=-0.0274, interpMode=RCIM_Cubic, leaveTangent=-0.0274, tangentMode=RCTM_User, time=0.9598, value=1}
          <at 1 1 RCIM_Cubic>
         ]
        }}})
    )
   ) {module=ScaleSpriteSize002}
   &DynamicMaterialParameters(
    Index 0 Param 1=1
    Index 0 Param 2=&FloatFromCurve(FloatCurve=DataInterface() {propertyValues {Curve {keys [<at 0 1>,<at 0.5 1>]}}})
    Index 0 Param 3=1
    Index 0 Param 4=&FloatFromCurve(FloatCurve=DataInterface() {propertyValues {Curve {keys [<at 0.8 0>,<at 1 1>]}}})
    Module.Index0Float0Write=true
    Module.Index0Float1Write=true
    Module.Index0Float2Write=false
    Module.Index0Float3Write=true
   )
   /SetVariables(
    Particles.Linear Color_Halo:LinearColor=&MakeLinearColorFromVectorAndFloat(
     "Vector (RGB)"=&Multiply_VectorByFloat(
      Vector=<0 0.5 1>
      Float=&FloatFromCurve(FloatCurve=DataInterface() {propertyValues {Curve {keys [<at 0 0.5>,<at 0.6 0.5>,<at 0.7 1>]}}})
     )
     "Float (Alpha)"=0.1
    )
    Particles.SpriteSize_Halo:Vector2f=&Multiply_Vector2D_ByFloat(
     Float=&FloatFromCurve(
      FloatCurve=DataInterface() {propertyValues {Curve {
         defaultValue=0
         keys [
          <at 0.6 0 RCIM_Cubic>
          {
           arriveTangent=1.4749
           arriveTangentWeight=0.1836
           interpMode=RCIM_Cubic
           leaveTangent=1.4749
           tangentMode=RCTM_User
           time=0.6496
           value=0.875
          }
          {arriveTangent=0.0101, arriveTangentWeight=0.1034, interpMode=RCIM_Cubic, leaveTangent=0.0101, tangentMode=RCTM_User, time=0.95, value=1}
         ]
        }}}
      Scale Curve=1.25
     )
     Vector2D=<50 50>
    )
   ) {isSetParameters=true}
  ]
  rendererClasses [&NiagaraSpriteRendererProperties]
  renderers [{rendererClass=&NiagaraSpriteRendererProperties}]
 }
}
summary {
 systemName=NS_ReactFX_PolarBlind2
 emitters [
  {emitterName=Intro_Heart, rendererClasses [&NiagaraSpriteRendererProperties]}
  {emitterName=Intro_Arrow, rendererClasses [&NiagaraSpriteRendererProperties]}
  {emitterName=Intro_FierySkull, rendererClasses [&NiagaraSpriteRendererProperties]}
  {emitterName=Intro_Burst, rendererClasses [&NiagaraSpriteRendererProperties]}
  {emitterName=Burst_Flame, rendererClasses [&NiagaraSpriteRendererProperties]}
  {emitterName=Burst_Embers, rendererClasses [&NiagaraSpriteRendererProperties]}
 ]
 userVariables [
  User.bIsFrontEnd {type=NiagaraBool}
  User.bIsFrontEndPreview {type=NiagaraBool}
 ]
}
