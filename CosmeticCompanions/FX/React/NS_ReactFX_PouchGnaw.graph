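# Niagara reaction effect for a cosmetic companion, named PouchGnaw, that plays a timed sprite and mesh sequence with a shared interaction sound.
# A Wheel sprite pops up with squash-stretch, then shakes, layered with spinning blur and rotating meshes.
# The Heart emitter plays the audio and animates a shrinking, camera-locked sprite, while a delayed debris burst and flash finish the sequence.
# Helper inputs adjust for front end and time of day; Wheel001 is disabled.
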NiagaraSystem /CosmeticCompanions/FX/React/NS_ReactFX_PouchGnaw
dependencies {
 usedDataInterfaces [
  &NiagaraDataInterfaceCurve
  &NiagaraDataInterfaceVector2DCurve
  &NiagaraDataInterfaceSpriteRendererInfo
  &NiagaraDataInterfaceAudioPlayer
  &NiagaraDataInterfaceCamera
  &NiagaraDataInterfaceArrayFloat
 ]
 usedDynamicInputs [
  &MergeHelper_FloatSwitch
  &SpawnGroup
  &Multiply_Vector
 ]
 usedModules [
  &self:SystemSpawnScript.SystemScriptSource.SystemScriptGraph.NiagaraNodeAssignment_7.SetVariables_#1
  &SystemState
 ]
}
emitters {
 Debris001 {
  emitterUpdateScript [
   &EmitterState()
   &SpawnBurst_Instantaneous(Spawn Count=14, Spawn Time=1.25, Use Spawn Probability=false, UseLoopCountLimit=false)
  ]
  particleSpawnScript [
   &InitializeParticle(
    Color Mode=::NewEnumerator0
    Lifetime Max=0.75
    Lifetime Min=0.1
    Lifetime Mode=::NewEnumerator1
    Position Offset=$System.BurstPosition
    Sprite Rotation Mode=::NewEnumerator2
    Sprite Size Mode=::NewEnumerator3
    Uniform Sprite Size=20
    UsePositionOffset=true
    Write Material Random=true
   )
   &PartitionParticles(
    Particles Per Partition=1
    Partition Count Mode:ENiagara_InfiniteLoopDuration=::NewEnumerator1
    Partition Mode:ENiagara_PartitionMode_v2=::NewEnumerator1
   )
   &ShapeLocation(MaskSpawnGroup=false, Override Seed=false, Sphere Radius=3, Sphere Surface Distribution=1, UseOwnerScale=true)
   &AddVelocity(
    Module.Use Falloff Distance=false
    Override Seed=false
    Velocity Mode=::NewEnumerator1
    Velocity Origin=$Particles.ShapeLocation.ShapePosition
    Velocity Speed=&RandomRangeFloat(Minimum=80, Maximum=100)
   )
   &SubUVAnimation(
    Direct Frame Index:NiagaraFloat=$Particles.PartitionParticles.NormalizedPartition
    SubUV Animation Mode:ENiagara_SubUVLookupModeV2=::NewEnumerator5
   )
  ]
  particleUpdateScript [
   &ParticleState()
   &SpriteRotationRate(Rotation Rate=&RandomRangeFloat(Minimum=-1000, Maximum=1000))
   &ScaleSpriteSize(Uniform Curve Sprite Scale=DataInterface() {propertyValues {Curve {
       defaultValue=0
       keys [
        {arriveTangent=9.0725, interpMode=RCIM_Cubic, leaveTangent=9.0725, tangentMode=RCTM_User, time=0, value=0}
        {arriveTangent=-0.054, interpMode=RCIM_Cubic, leaveTangent=-0.054, tangentMode=RCTM_User, time=0.1408, value=0.9847}
        {arriveTangent=-0.0405, interpMode=RCIM_Cubic, leaveTangent=-0.0405, tangentMode=RCTM_User, time=1, value=0}
       ]
      }}})
   &GravityForce(Gravity=&RandomRangeVector(Minimum=<0 0 -100>, Maximum=<0 0 -250>))
   &Drag(Drag=&RandomRangeFloat(Minimum=0.8, Maximum=1.2), Use Linear Drag=true, Use Rotational Drag=true)
   &SolveForcesAndVelocity(Clamp Velocity=false, Limit Acceleration=false)
  ]
  rendererClasses [&NiagaraSpriteRendererProperties]
  renderers [{rendererClass=&NiagaraSpriteRendererProperties}]
 }
 Heart {
  emitterSpawnScript [
   &InitializeAudioPlayer(Audio Player=DataInterface() {propertyValues {
      Attenuation=null
      Concurrency=null
      MaxPlaysPerTick=10
      SoundToPlay=/CosmeticCompanions/Assets/Audio_Shared/Interaction/Petting/SC_Mimosa_Interaction_React_Default
      bAllowLoopingOneShotSounds=false
      bLimitPlaysPerTick=true
      bOnlyActiveDuringGameplay=false
      bStopWhenComponentIsDestroyed=true
      ConfigurationUserParameter {parameter {name=None, typeDefHandle {registeredTypeIndex=1}, varData []}}
      ParameterNames []
     }})
   /Niagara/Modules/Audio/PlayAudio() {module=PlayAudio001}
  ]
  emitterUpdateScript [
   &EmitterState(Inactive Response=::NewEnumerator1, Life Cycle Mode=::NewEnumerator1, Loop Behavior=::NewEnumerator1, Loop Duration=1.25, UseLoopDelay=false)
   &SpawnBurst_Instantaneous(Spawn Time=0.05, Use Spawn Probability=false, UseLoopCountLimit=false) {module=SpawnBurst_Instantaneous001}
  ]
  particleSpawnScript [
   &InitializeParticle(
    Lifetime=1.25
    Mesh Scale Mode=::NewEnumerator3
    Mesh Uniform Scale=5
    Ribbon Width=10
    Ribbon Width Mode=::NewEnumerator1
    Sprite Size Mode=::NewEnumerator3
    Uniform Sprite Size=18
    UsePositionOffset=false
    Write Material Random=true
   )
   /SetVariables(Particles.VisibilityTag=$Emitter.SpawnGroup) {isSetParameters=true}
   /SetVariables(Particles.SortingLayer:NiagaraFloat=&SelectFloatFromArray(Direct Array Index=$Particles.VisibilityTag, Float Selection Array=DataInterface() {propertyValues {FloatData [-5,-5]}})) {isSetParameters=true}
   &CameraOffset(Camera Offset Amount=&MaskFloat(Mask Source=1, Mask Parameter=$Particles.VisibilityTag, Mask=0, Masked Float=0.1, Passthrough Float=-0.1))
  ]
  particleUpdateScript [
   &ParticleState(Kill Particles When Lifetime Has Elapsed=false, Loop Particles Lifetime=true)
   &ScaleSpriteSize(
    Scale Factor=&MakeVector2D(
     X=&FloatFromCurve(FloatCurve=DataInterface() {propertyValues {Curve {
         defaultValue=0
         keys [
          {
           arriveTangent=-0.0148
           arriveTangentWeight=0.296
           interpMode=RCIM_Cubic
           leaveTangent=-0.0148
           leaveTangentWeight=0.296
           tangentMode=RCTM_User
           tangentWeightMode=RCTWM_WeightedBoth
           time=0
           value=1
          }
          {
           arriveTangent=-24.7764
           arriveTangentWeight=0.8868
           interpMode=RCIM_Cubic
           leaveTangent=63.8831
           leaveTangentWeight=0.1496
           tangentMode=RCTM_Break
           tangentWeightMode=RCTWM_WeightedBoth
           time=0.3
           value=0
          }
         ]
        }}})
     Y=&FloatFromCurve(FloatCurve=DataInterface() {propertyValues {Curve {
         defaultValue=0
         keys [
          {
           arriveTangent=-0.0148
           arriveTangentWeight=0.296
           interpMode=RCIM_Cubic
           leaveTangent=-0.0148
           leaveTangentWeight=0.296
           tangentMode=RCTM_User
           tangentWeightMode=RCTWM_WeightedBoth
           time=0
           value=1
          }
          {
           arriveTangent=-24.7764
           arriveTangentWeight=0.8868
           interpMode=RCIM_Cubic
           leaveTangent=63.8831
           leaveTangentWeight=0.1496
           tangentMode=RCTM_Break
           tangentWeightMode=RCTWM_WeightedBoth
           time=0.3
           value=0
          }
         ]
        }}})
    )
    Scale Sprite Size Mode=::NewEnumerator2
   ) {module=ScaleSpriteSize001}
   &OffsetPosition(Position Offset=&MakeVector(
     X=0
     Y=&FloatFromCurve(
      FloatCurve=DataInterface() {propertyValues {Curve {keys [
          {arriveTangent=623.2311, interpMode=RCIM_Cubic, time=0, value=10}
          {arriveTangent=-34.2857, interpMode=RCIM_Cubic, leaveTangent=-34.2857, time=0.15, value=15}
          {interpMode=RCIM_Cubic, leaveTangent=-8.3162, time=0.35, value=-2}
         ]}}}
      CurveIndex=$Emitter.NormalizedLoopAge
     )
     Z=&FloatFromCurve(
      FloatCurve=DataInterface() {propertyValues {Curve {keys [
          {arriveTangent=-217.9887, interpMode=RCIM_Cubic, leaveTangent=-217.9887, tangentMode=RCTM_User, time=0, value=25}
          {arriveTangent=3.7992, interpMode=RCIM_Cubic, leaveTangent=3.7992, tangentMode=RCTM_User, time=0.3, value=-15}
         ]}}}
      CurveIndex=$Emitter.NormalizedLoopAge
     )
    ))
   &OffsetPosition(Position Offset=&MakeVector(
     X=0
     Y=&Lerp_Float(
      A=0
      B=&Waveform("Waveform [1]"=::NewEnumerator0, "[1] Frequency"=10, "[1] Amplitude Min/Max"=<-1 1>, "[1] Scale"=10)
      Alpha=&FloatFromCurve(
       FloatCurve=DataInterface() {propertyValues {Curve {keys [
           <at 0 0>
           {arriveTangent=0.0852, interpMode=RCIM_Cubic, leaveTangent=0.0852, time=0.4255, value=0}
           {arriveTangent=0.2112, interpMode=RCIM_Cubic, leaveTangent=0.2112, time=0.5663, value=0.0482}
           {arriveTangent=0.0151, interpMode=RCIM_Cubic, leaveTangent=0.0151, tangentMode=RCTM_User, time=0.7817, value=0.0752}
          ]}}}
       CurveIndex=$Emitter.NormalizedLoopAge
      )
     )
     Z=0
    )) {module=OffsetPosition001}
   &OffsetPosition(Position Offset=$System.BurstPosition) {enabled=false, module=OffsetPosition002}
   &self:GimbalLock_06(Gimbal Lock Direction:NiagaraPosition=&ReturnCameraVector(Camera Vectors=::NewEnumerator4, Coordinate Space=::Local, Origin Position=$Engine.Emitter.SimulationPosition)) {module=GimbalLock_06}
  ]
  rendererClasses [&NiagaraSpriteRendererProperties]
  renderers [{rendererClass=&NiagaraSpriteRendererProperties}]
 }
 LWE_Corona001 {rendererClasses [], renderers []}
 LWE_Rays {rendererClasses [], renderers []}
 Minimal {
  emitterUpdateScript [
   &EmitterState()
   &SpawnBurst_Instantaneous(Spawn Time=1.25, Use Spawn Probability=false, UseLoopCountLimit=false)
  ]
  particleSpawnScript [&InitializeParticle(
    Color Mode=::NewEnumerator0
    Lifetime=0.4
    Sprite Size Mode=::NewEnumerator3
    Uniform Sprite Size=35
    UsePositionOffset=false
    Write Material Random=true
   )]
  particleUpdateScript [
   &ParticleState()
   &ScaleSpriteSize(Uniform Curve Sprite Scale=DataInterface() {propertyValues {Curve {
       defaultValue=0
       keys [
        {arriveTangent=6.951, interpMode=RCIM_Cubic, leaveTangent=6.951, tangentMode=RCTM_User, time=0, value=0}
        {interpMode=RCIM_Cubic, tangentMode=RCTM_User, time=0.5, value=1}
        {arriveTangent=-0.8112, tangentMode=RCTM_User, time=1, value=0}
       ]
      }}})
   &OffsetPosition(Position Offset=&FN_FrontEndVector3Modifier(Gameplay=<0 0 -10>, Front End=<0 0 -10>))
  ]
  rendererClasses [&NiagaraSpriteRendererProperties]
  renderers [{rendererClass=&NiagaraSpriteRendererProperties}]
 }
 Minimal001 {
  emitterUpdateScript [
   &EmitterState(
    Life Cycle Mode=::NewEnumerator1
    Loop Behavior=::NewEnumerator1
    Loop Delay=0.3
    Loop Duration=0.7
    Scalability Mode=::NewEnumerator1
    UseLoopDelay=true
   )
   &SpawnBurst_Instantaneous(Spawn Count=2, Use Spawn Probability=false, UseLoopCountLimit=false)
   &SpawnRate(SpawnRate=6, Use Spawn Probability=false)
  ]
  particleSpawnScript [
   &InitializeParticle(
    Color=&RandomRangeLinearColor(Minimum=<rgba 0.3017 0.3113 0.3652 0.5>, Maximum=<rgba 0 0.0688 0.4531 0.8>)
    Lifetime=0.25
    Mesh Scale Mode=::NewEnumerator3
    Mesh Uniform Scale=0.1025
    UsePositionOffset=false
    Write Material Random=true
   )
   &PartitionParticles()
   &InitialMeshOrientation(
    Mesh Axis To Orient=<0 1 0>
    Mesh Orientation Mode=::NewEnumerator2
    Rotation=&MakeVector(X=0, Y=&Multiply_FloatByInt(Float=0.333, Integer=$Particles.PartitionParticles.Partition), Z=0)
    Use Rotation Vector=true
   )
   &CameraOffset(Camera Offset Amount=5)
  ]
  particleUpdateScript [
   &ParticleState()
   &SkeletalMeshLocation(Bone Sampling Mode=::NewEnumerator4, MaskSpawnGroup=false, Sampled Position Offset=<10 0 0>, Skeletal Mesh=$System.Skeletal Mesh) {enabled=false}
   &UpdateMeshOrientation(
    Rotation Rate=&FloatFromCurve(
     FloatCurve=DataInterface() {propertyValues {Curve {defaultValue=1, keys [{tangentMode=RCTM_User, time=0, value=0}, {tangentMode=RCTM_User, time=1, value=1}]}}}
     Scale Curve=25
    )
    Rotation Vector=<0 1 0>
   )
   &ScaleColor(
    Scale Alpha=&FloatFromCurve(
     FloatCurve=DataInterface() {propertyValues {Curve {keys [
         <at 0 0 RCIM_Cubic>
         {arriveTangentWeight=0.0373, interpMode=RCIM_Cubic, leaveTangentWeight=0.3896, tangentMode=RCTM_User, time=0.2, value=1}
         {interpMode=RCIM_Cubic, tangentMode=RCTM_User, time=0.8, value=1}
         <at 1 0>
        ]}}}
     Scale Curve=0.1
    )
    Scale RGB=&MakeVectorFromLinearColorRGB(Color=&FNTimeOfDayColor(
      Night Color=<rgba 1.8 1.8 1.8 1>
      Morning Color=<rgba 1.35 1.35 1.35 1>
      Day Color=<rgba 1.45 1.45 1.45 1>
      Evening Color=<rgba 1.35 1.35 1.35 1>
     ))
    ScaleA=true
    ScaleRGB=true
   )
   &OffsetPosition(Position Offset=<0 0 -12>)
   /SetVariables(Particles.Position=&AddVectorToPosition(
     Position=$Particles.OffsetPosition.Position
     Vector=&MakeVector(
      X=&Waveform(
       "Waveform [1]"=::NewEnumerator0
       "[1] Frequency"=&FloatFromCurve(FloatCurve=DataInterface() {propertyValues {Curve {keys [<at 0.6 0>,<at 0.7 1>]}}}, Scale Curve=8)
       "[1] Amplitude Min/Max"=<-1 1>
      )
      Y=0
      Z=&Waveform(
       "Waveform [1]"=::NewEnumerator0
       "[1] Frequency"=&FloatFromCurve(FloatCurve=DataInterface() {propertyValues {Curve {keys [<at 0.6 0>,<at 0.7 1>]}}}, Scale Curve=8)
       "[1] Amplitude Min/Max"=<-1 1>
      )
     )
    )) {enabled=false, isSetParameters=true}
   &ScaleColor(
    Scale Alpha=&FloatFromCurve(
     FloatCurve=DataInterface() {propertyValues {Curve {
        defaultValue=1
        keys [
         {arriveTangent=4.1785, interpMode=RCIM_Cubic, leaveTangent=4.1785, tangentMode=RCTM_User, time=0, value=0}
         {arriveTangent=1.1179, interpMode=RCIM_Cubic, leaveTangent=1.1179, tangentMode=RCTM_User, time=0.2042, value=1}
         {arriveTangent=0.0224, interpMode=RCIM_Cubic, leaveTangent=0.0224, tangentMode=RCTM_User, time=1, value=1}
        ]
       }}}
     CurveIndex=$Emitter.NormalizedLoopAge
    )
    ScaleA=true
    ScaleRGB=true
   ) {module=ScaleColor002}
  ]
  rendererClasses [&NiagaraMeshRendererProperties]
  renderers [{rendererClass=&NiagaraMeshRendererProperties, rendererIndex=0}]
 }
 SpinBlur {
  emitterUpdateScript [
   &EmitterState()
   &SpawnBurst_Instantaneous(Spawn Time=0.32, Use Spawn Probability=false, UseLoopCountLimit=false)
  ]
  particleSpawnScript [&InitializeParticle(
    Color=<rgba 0.4 0.4 0.4 1>
    Lifetime=0.9
    Position Offset=$System.BurstPosition
    Sprite Size Mode=::NewEnumerator3
    Uniform Sprite Size=28
    UsePositionOffset=false
    Write Material Random=true
   )]
  particleUpdateScript [
   &ParticleState()
   &SpriteRotationRate(Rotation Rate=&FloatFromCurve(
     FloatCurve=DataInterface() {propertyValues {Curve {defaultValue=1, keys [{tangentMode=RCTM_User, time=0, value=0}, {tangentMode=RCTM_User, time=0.1, value=5}]}}}
     Scale Curve=-250
    ))
   &ScaleSpriteSize(Uniform Curve Sprite Scale=DataInterface() {propertyValues {Curve {keys [<at 0 0 RCIM_Cubic>,<at 0.225 1 RCIM_Cubic>]}}}) {enabled=false}
   &ScaleSpriteSize(
    Non-Uniform Curve Sprite Scale=DataInterface() {propertyValues {
      XCurve {keys [<at 0 0 RCIM_Cubic>,<at 0.15 1 RCIM_Cubic>]}
      YCurve {keys [
        <at 0 0 RCIM_Cubic>
        {arriveTangent=-0.033, interpMode=RCIM_Cubic, leaveTangent=-0.033, tangentMode=RCTM_User, time=0.075, value=1.75}
        <at 0.15 1 RCIM_Cubic>
       ]}
     }}
    Scale Sprite Size Mode=::NewEnumerator3
   ) {enabled=false, module=ScaleSpriteSize001}
   &SubUVAnimation(
    "Sprite Renderer "=DataInterface() {propertyValues {SpriteRenderer=&self:SpinBlur_0.NiagaraSpriteRendererProperties_0}}
    SubUV Lookup Index:NiagaraFloat=&FloatFromCurve(FloatCurve=DataInterface() {propertyValues {Curve {keys [{arriveTangent=14.8245, interpMode=RCIM_Cubic, leaveTangent=14.8245, tangentMode=RCTM_User, time=0.15, value=0}, <at 0.25 1 RCIM_Cubic>]}}})
    UseEndFrame=false
    UseStartFrame=false
   ) {enabled=false}
   &OffsetPosition(Position Offset=&FN_FrontEndVector3Modifier(Gameplay=<1 -0.75 -12>, Front End=<1 -0.75 -12>))
   &ScaleColor(
    Scale Alpha=&FloatFromCurve(FloatCurve=DataInterface() {propertyValues {Curve {keys [<at 0 1>,<at 0.8938 1>,<at 1 0>]}}})
    Scale RGB=&VectorFromFloat(Value=&FNTimeOfDayScalar(Night Scale=1.25, Morning Scale=1.5, Day Scale=1.2, Evening Scale=1.5))
    ScaleA=false
    ScaleRGB=true
   )
   /SetVariables(Particles.Position=&AddVectorToPosition(
     Position=$Particles.OffsetPosition.Position
     Vector=&MakeVector(
      X=&Waveform(
       "Waveform [1]"=::NewEnumerator0
       "[1] Frequency"=&FloatFromCurve(FloatCurve=DataInterface() {propertyValues {Curve {keys [<at 0.6 0>,<at 0.7 1>]}}}, Scale Curve=8)
       "[1] Amplitude Min/Max"=<-1 1>
      )
      Y=0
      Z=&Waveform(
       "Waveform [1]"=::NewEnumerator0
       "[1] Frequency"=&FloatFromCurve(FloatCurve=DataInterface() {propertyValues {Curve {keys [<at 0.6 0>,<at 0.7 1>]}}}, Scale Curve=8)
       "[1] Amplitude Min/Max"=<-1 1>
      )
     )
    )) {enabled=false, isSetParameters=true, module=SetVariables_#2}
  ]
  rendererClasses [&NiagaraSpriteRendererProperties]
  renderers [{rendererClass=&NiagaraSpriteRendererProperties}]
 }
 Wheel {
  emitterUpdateScript [
   &EmitterState()
   &SpawnBurst_Instantaneous(Use Spawn Probability=false, UseLoopCountLimit=false)
  ]
  particleSpawnScript [
   &InitializeParticle(
    Lifetime=1.25
    Position Offset=$System.BurstPosition
    Sprite Size Mode=::NewEnumerator3
    Uniform Sprite Size=28
    UsePositionOffset=false
    Write Material Random=true
   )
   &CameraOffset(Camera Offset Amount=5) {enabled=false}
  ]
  particleUpdateScript [
   &ParticleState()
   &ScaleSpriteSize(Uniform Curve Sprite Scale=DataInterface() {propertyValues {Curve {keys [<at 0 0 RCIM_Cubic>,<at 0.225 1 RCIM_Cubic>]}}}) {enabled=false}
   &ScaleSpriteSize(
    Non-Uniform Curve Sprite Scale=DataInterface() {propertyValues {
      XCurve {keys [<at 0 0 RCIM_Cubic>,<at 0.15 1 RCIM_Cubic>]}
      YCurve {keys [
        <at 0 0 RCIM_Cubic>
        {arriveTangent=-0.033, interpMode=RCIM_Cubic, leaveTangent=-0.033, tangentMode=RCTM_User, time=0.075, value=1.75}
        <at 0.15 1 RCIM_Cubic>
       ]}
     }}
    Scale Sprite Size Mode=::NewEnumerator3
   ) {module=ScaleSpriteSize001}
   &SubUVAnimation(
    "Sprite Renderer "=DataInterface() {propertyValues {SpriteRenderer=&self:Wheel_0.NiagaraSpriteRendererProperties_0}}
    SubUV Lookup Index:NiagaraFloat=&FloatFromCurve(FloatCurve=DataInterface() {propertyValues {Curve {keys [{arriveTangent=14.8245, interpMode=RCIM_Cubic, leaveTangent=14.8245, tangentMode=RCTM_User, time=0.15, value=0}, <at 0.25 1 RCIM_Cubic>]}}})
    UseEndFrame=false
    UseStartFrame=false
   ) {enabled=false}
   &OffsetPosition(Position Offset=&Lerp_Vector(
     A=&FN_FrontEndVector3Modifier(Gameplay=<0 0 -80>, Front End=<0 0 -80>)
     B=&FN_FrontEndVector3Modifier(Gameplay=<0 0 -15>, Front End=<0 0 -15>)
     Alpha=&FloatFromCurve(FloatCurve=DataInterface() {propertyValues {Curve {
         defaultValue=1
         keys [
          {arriveTangent=7.9822, interpMode=RCIM_Cubic, leaveTangent=7.9822, tangentMode=RCTM_User, time=0, value=0}
          {interpMode=RCIM_Cubic, tangentMode=RCTM_User, time=0.225, value=1}
         ]
        }}})
    ))
   &ScaleColor(
    Scale Alpha=&FloatFromCurve(FloatCurve=DataInterface() {propertyValues {Curve {keys [<at 0 1>,<at 0.8938 1>,<at 1 0>]}}})
    ScaleA=false
    ScaleRGB=true
   )
   /SetVariables(Particles.Position=&AddVectorToPosition(
     Position=$Particles.OffsetPosition.Position
     Vector=&Multiply_VectorByFloat(
      Vector=&MakeVector(
       X=&Waveform(
        "Waveform [1]"=::NewEnumerator0
        "[1] Frequency"=&FloatFromCurve(FloatCurve=DataInterface() {propertyValues {Curve {keys [<at 0.6 0>,<at 0.7 1>]}}}, Scale Curve=8)
        "[1] Amplitude Min/Max"=<-1 1>
       )
       Y=0
       Z=&Waveform(
        "Waveform [1]"=::NewEnumerator0
        "[1] Frequency"=&FloatFromCurve(FloatCurve=DataInterface() {propertyValues {Curve {keys [<at 0.6 0>,<at 0.7 1>]}}}, Scale Curve=8)
        "[1] Amplitude Min/Max"=<-1 1>
       )
      )
      Float=&FN_FrontEndFloatModifier(Gameplay=1, Front End=0.15)
     )
    )) {isSetParameters=true, module=SetVariables_#2}
  ]
  rendererClasses [&NiagaraSpriteRendererProperties]
  renderers [
   {rendererClass=&NiagaraSpriteRendererProperties}
   {rendererClass=&NiagaraSpriteRendererProperties, rendererIndex=1}
  ]
 }
 Wheel001 {
  bEnabled=false
  emitterSpawnScript [
   &InitializeAudioPlayer(Audio Player=DataInterface() {propertyValues {
      Attenuation=null
      Concurrency=null
      MaxPlaysPerTick=10
      SoundToPlay=&SC_Mimosa_Interaction_React_Default
      bAllowLoopingOneShotSounds=false
      bLimitPlaysPerTick=true
      bOnlyActiveDuringGameplay=false
      bStopWhenComponentIsDestroyed=true
      ConfigurationUserParameter {parameter {name=None, typeDefHandle {registeredTypeIndex=1}, varData []}}
      ParameterNames []
     }})
   &PlayAudio() {enabled=false, module=PlayAudio001}
  ]
  emitterUpdateScript [
   &EmitterState()
   &SpawnBurst_Instantaneous(Use Spawn Probability=false, UseLoopCountLimit=false)
  ]
  particleSpawnScript [
   &InitializeParticle(
    Color=&MakeCustomLinearColorFromBool(
     Bool=$User.bIsFrontEndPreview
     True Color=<rgba 0.5 0.5 0.5 1>
     False Color=&FNTimeOfDayColor(
      Night Color=<rgba 0.5366 0.5587 0.625 1>
      Morning Color=<rgba 0.7188 0.649 0.577 1>
      Day Color=<rgba 0.8385 0.8385 0.8385 1>
      Evening Color=<rgba 0.7188 0.649 0.577 1>
     )
    )
    Position Offset=$System.BurstPosition
    Sprite Size Mode=::NewEnumerator3
    Uniform Sprite Size=12
    UsePositionOffset=false
    Write Material Random=true
   )
   &CameraOffset(Camera Offset Amount=10)
  ]
  particleUpdateScript [
   &ParticleState()
   &ScaleSpriteSize(Uniform Curve Sprite Scale=DataInterface() {propertyValues {Curve {keys [<at 0 0 RCIM_Cubic>,<at 0.225 1 RCIM_Cubic>]}}}) {enabled=false}
   &ScaleSpriteSize(
    Scale Sprite Size Mode=::NewEnumerator0
    Uniform Scale Factor=&FloatFromCurve(FloatCurve=DataInterface() {propertyValues {Curve {
        defaultValue=1
        keys [
         {interpMode=RCIM_Cubic, tangentMode=RCTM_User, time=0.1, value=1}
         {tangentMode=RCTM_User, time=0.3, value=0}
         {interpMode=RCIM_Cubic, tangentMode=RCTM_User, time=1, value=0}
        ]
       }}})
   ) {module=ScaleSpriteSize001}
   &SubUVAnimation(
    "Sprite Renderer "=DataInterface() {propertyValues {SpriteRenderer=&self:Wheel001_1.NiagaraSpriteRendererProperties_0}}
    SubUV Lookup Index:NiagaraFloat=&FloatFromCurve(FloatCurve=DataInterface() {propertyValues {Curve {keys [{arriveTangent=14.8245, interpMode=RCIM_Cubic, leaveTangent=14.8245, tangentMode=RCTM_User, time=0.15, value=0}, <at 0.25 1 RCIM_Cubic>]}}})
    UseEndFrame=false
    UseStartFrame=false
   ) {enabled=false}
   &OffsetPosition(Position Offset=&MakeVector(
     X=0
     Y=&FloatFromCurve(
      FloatCurve=DataInterface() {propertyValues {Curve {keys [
          <at 0.2891 -127 RCIM_Cubic>
          {arriveTangent=523.8333, interpMode=RCIM_Cubic, leaveTangent=523.8333, tangentMode=RCTM_User, time=0.4902, value=-25}
         ]}}}
      CurveIndex=$Emitter.NormalizedLoopAge
     )
     Z=&FloatFromCurve(
      FloatCurve=DataInterface() {propertyValues {Curve {keys [<at 0.2839 90>, {arriveTangent=-237.6561, interpMode=RCIM_Cubic, leaveTangent=-237.6561, time=0.4033, value=37.9158}, <at 0.4937 40.1355>]}}}
      CurveIndex=$Emitter.NormalizedLoopAge
     )
    )) {enabled=false}
   &ScaleColor(
    Scale Alpha=&FloatFromCurve(FloatCurve=DataInterface() {propertyValues {Curve {keys [<at 0 1>,<at 0.8938 1>,<at 1 0>]}}})
    Scale RGB=&VectorFromFloat(Value=2.25)
    ScaleA=true
    ScaleRGB=true
   )
   /SetVariables(Particles.Position=&MakeVector(
     X=0
     Y=&FloatFromCurve(
      FloatCurve=DataInterface() {propertyValues {Curve {keys [
          {arriveTangent=-1030.2709, interpMode=RCIM_Cubic, leaveTangent=-1030.2709, tangentMode=RCTM_User, time=0, value=-50}
          {arriveTangent=-201.5283, interpMode=RCIM_Cubic, leaveTangent=-201.5283, tangentMode=RCTM_User, time=0.2, value=0}
         ]}}}
      CurveIndex=$Emitter.NormalizedLoopAge
     )
     Z=&FloatFromCurve(
      FloatCurve=DataInterface() {propertyValues {Curve {keys [{arriveTangent=-381.0104, interpMode=RCIM_Cubic, leaveTangent=-381.0104, tangentMode=RCTM_User, time=0, value=-15}, <at 0.25 -35>]}}}
      CurveIndex=$Emitter.NormalizedLoopAge
     )
    )) {isSetParameters=true, module=SetVariables_#2}
   &self:GimbalLock_05(Gimbal Lock Direction:NiagaraPosition=&ReturnCameraVector(Camera Vectors=::NewEnumerator4, Coordinate Space=::Local, Origin Position=$Engine.Emitter.SimulationPosition)) {module=GimbalLock_05}
  ]
  rendererClasses [&NiagaraSpriteRendererProperties]
  renderers [{rendererClass=&NiagaraSpriteRendererProperties}]
 }
}
summary {
 systemName=NS_ReactFX_PouchGnaw
 emitters [
  {emitterName=LWE_Rays, rendererClasses []}
  {emitterName=Wheel, rendererClasses [&NiagaraSpriteRendererProperties]}
  {emitterName=LWE_Corona001, rendererClasses []}
  {emitterName=Debris001, rendererClasses [&NiagaraSpriteRendererProperties]}
  {bEnabled=false, emitterName=Wheel001, rendererClasses [&NiagaraSpriteRendererProperties]}
  {emitterName=Heart, rendererClasses [&NiagaraSpriteRendererProperties]}
  {emitterName=Minimal001, rendererClasses [&NiagaraMeshRendererProperties]}
  {emitterName=SpinBlur, rendererClasses [&NiagaraSpriteRendererProperties]}
  {emitterName=Minimal, rendererClasses [&NiagaraSpriteRendererProperties]}
 ]
 userVariables [
  User.bIsFrontEnd {type=NiagaraBool}
  User.bIsFrontEndPreview {type=NiagaraBool, defaultValue=true}
  User.ReverseTimeNS {type=NiagaraFloat, defaultValue=1, description=""}
 ]
}
