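# Niagara reaction effect for the PromDutch cosmetic companion, pairing a default petting-reaction sound with a heart burst.
# Intro plays the audio and animates a glowing sprite from HeartIntroSpawnPosition to BurstPosition over IntroLength, dimming color by time of day.
# SmallHearts bursts just before the intro ends, flinging randomly tinted sub-image sprites outward and up under gravity as they shrink and fade.
# LWE_AirGlow and LWE_Rays show no modules or renderers.
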NiagaraSystem /CosmeticCompanions/FX/React/NS_ReactFX_PromDutch
dependencies {
 usedDataInterfaces [&NiagaraDataInterfaceAudioPlayer, &NiagaraDataInterfaceCurve, &NiagaraDataInterfaceColorCurve]
 usedDynamicInputs [
  &MergeHelper_VectorSwitch
  &ReturnExecIndex
  &FN_FrontEndFloatModifier
  &FN_FrontEndVector3Modifier
 ]
 usedModules [
  &self:SystemSpawnScript.SystemScriptSource.SystemScriptGraph.NiagaraNodeAssignment_3.SetVariables_#1
  &self:SystemSpawnScript.SystemScriptSource.SystemScriptGraph.NiagaraNodeAssignment_0.SetVariables_#2
  &self:SystemSpawnScript.SystemScriptSource.SystemScriptGraph.NiagaraNodeAssignment_7.SetVariables_#3
  &self:SystemSpawnScript.SystemScriptSource.SystemScriptGraph.NiagaraNodeAssignment_8.SetVariables_#4
  &self:SystemSpawnScript.SystemScriptSource.SystemScriptGraph.NiagaraNodeAssignment_9.SetVariables_#5
  &self:SystemSpawnScript.SystemScriptSource.SystemScriptGraph.NiagaraNodeAssignment_10.SetVariables_#6
  &self:SystemSpawnScript.SystemScriptSource.SystemScriptGraph.NiagaraNodeAssignment_5.SetVariables_#7
  &self:SystemSpawnScript.SystemScriptSource.SystemScriptGraph.NiagaraNodeAssignment_6.SetVariables_#8
  &SystemState
  &self:SystemSpawnScript.SystemScriptSource.SystemScriptGraph.NiagaraNodeAssignment_2.SetVariables_#9
 ]
}
emitters {
 Intro {
  emitterSpawnScript [
   &InitializeAudioPlayer(Audio Player=DataInterface() {propertyValues {
      Attenuation=null
      Concurrency=null
      MaxPlaysPerTick=10
      SoundToPlay=/CosmeticCompanions/Assets/Audio_Shared/Interaction/Petting/SC_Mimosa_Interaction_React_Default
      bAllowLoopingOneShotSounds=false
      bLimitPlaysPerTick=true
      bOnlyActiveDuringGameplay=false
      bStopWhenComponentIsDestroyed=true
      ConfigurationUserParameter {parameter {name=None, typeDefHandle {registeredTypeIndex=1}, varData []}}
      ParameterNames []
     }})
   /Niagara/Modules/Audio/PlayAudio() {module=PlayAudio001}
  ]
  emitterUpdateScript [
   &EmitterState()
   &SpawnBurst_Instantaneous(Use Spawn Probability=false, UseLoopCountLimit=false)
  ]
  particleSpawnScript [
   &InitializeParticle(
    Lifetime=$System.IntroLength
    Sprite Rotation Mode=::NewEnumerator0
    Sprite Size Mode=::NewEnumerator3
    Uniform Sprite Size=10.5
    UsePositionOffset=false
    Write Material Random=true
   )
   /SetVariables(Particles.GlowSubImage=1) {isSetParameters=true}
  ]
  particleUpdateScript [
   &ParticleState()
   &ScaleSpriteSize(
    Scale Sprite Size Mode=::NewEnumerator0
    Uniform Scale Factor=&Add_Float(
     A=&FloatFromCurve(FloatCurve=DataInterface() {propertyValues {Curve {
         defaultValue=0.5
         keys [
          <at 0 0.2 RCIM_Cubic>
          {arriveTangent=0.375, interpMode=RCIM_Cubic, leaveTangent=0.375, time=0.3, value=0.5}
          {arriveTangent=-0.5031, interpMode=RCIM_Cubic, leaveTangent=-0.5031, tangentMode=RCTM_User, time=0.8, value=0.5}
          {arriveTangent=3.3877, interpMode=RCIM_Cubic, leaveTangent=3.3877, time=0.884, value=0.2079}
          {arriveTangent=-1.5179, interpMode=RCIM_Cubic, leaveTangent=-1.5179, tangentMode=RCTM_User, time=0.9444, value=0.9892}
          <at 1.0007 0.7861 RCIM_Cubic>
         ]
        }}})
     B=&MergeHelper_FloatSwitch(
      0=&FloatFromCurve(FloatCurve=DataInterface() {propertyValues {Curve {
          defaultValue=0
          keys [
           {arriveTangent=3.2329, interpMode=RCIM_Cubic, leaveTangent=3.2329, tangentMode=RCTM_User, time=0, value=0}
           {arriveTangent=0.6241, interpMode=RCIM_Cubic, leaveTangent=0.6241, tangentMode=RCTM_User, time=0.1569, value=0.2299}
           {arriveTangent=2.2355, interpMode=RCIM_Cubic, leaveTangent=2.2355, tangentMode=RCTM_User, time=0.4075, value=0.6225}
           {arriveTangent=3.5453, interpMode=RCIM_Cubic, leaveTangent=3.5453, time=0.5461, value=1.084}
           {arriveTangent=4.4217, interpMode=RCIM_Cubic, leaveTangent=4.4217, time=0.6608, value=1.5207}
           {arriveTangent=-0.9053, interpMode=RCIM_Cubic, leaveTangent=-0.9053, tangentMode=RCTM_User, time=0.7123, value=1.819}
           {arriveTangent=-2.847, interpMode=RCIM_Cubic, leaveTangent=-2.847, time=0.7406, value=1.9236}
           {arriveTangent=-0.1202, interpMode=RCIM_Cubic, leaveTangent=-0.1202, tangentMode=RCTM_User, time=1, value=1}
          ]
         }}})
      1=&FloatFromCurve(
       FloatCurve=DataInterface() {propertyValues {Curve {
          defaultValue=0
          keys [
           {arriveTangent=0.087, interpMode=RCIM_Cubic, leaveTangent=0.087, tangentMode=RCTM_User, time=0, value=0}
           {arriveTangent=1.0515, interpMode=RCIM_Cubic, leaveTangent=1.0515, time=0.3888, value=0.1798}
           {arriveTangent=2.4243, interpMode=RCIM_Cubic, leaveTangent=2.4243, tangentMode=RCTM_User, time=0.5921, value=0.6225}
           {arriveTangent=1.5149, interpMode=RCIM_Cubic, leaveTangent=1.5149, tangentMode=RCTM_User, time=0.701, value=1.2762}
           {arriveTangent=-0.1202, interpMode=RCIM_Cubic, leaveTangent=-0.1202, tangentMode=RCTM_User, time=1, value=0.7243}
          ]
         }}}
       Scale Curve=0.7
      )
      2=?
      3=?
      Index=$Particles.PartitionParticles.Partition
      Num=::NewEnumerator0
     )
    )
   ) {module=ScaleSpriteSize002}
   &ScaleColor(
    Scale Alpha=&FloatFromCurve(FloatCurve=DataInterface() {propertyValues {Curve {
        defaultValue=0
        keys [
         {arriveTangent=9.0725, interpMode=RCIM_Cubic, leaveTangent=9.0725, tangentMode=RCTM_User, time=0, value=0}
         {arriveTangent=-0.2441, leaveTangent=-0.2441, time=0.15, value=1}
         {arriveTangent=-1.146, leaveTangent=-1.146, time=0.97, value=1}
         {arriveTangent=-0.0405, interpMode=RCIM_Cubic, leaveTangent=-0.0405, tangentMode=RCTM_User, time=0.971, value=1}
        ]
       }}})
    Scale RGB=&VectorFromFloat(Value=&FNTimeOfDayScalar(Night Scale=0.65, Morning Scale=0.9, Evening Scale=0.85))
    ScaleA=true
    ScaleRGB=true
   )
   &OffsetPosition(Position Offset=&Lerp_Vector(
     A=$System.HeartIntroSpawnPosition
     B=$System.BurstPosition
     Alpha=&FloatFromCurve(FloatCurve=DataInterface() {propertyValues {Curve {keys [
          {arriveTangent=3.2087, interpMode=RCIM_Cubic, leaveTangent=3.2087, tangentMode=RCTM_User, time=0, value=0}
          {arriveTangent=-0.1074, tangentMode=RCTM_User, time=1, value=1}
         ]}}})
    ))
   &DynamicMaterialParameters(
    Index 0 Param 2=1
    Module.Index0Float0Write=true
    Module.Index0Float1Write=true
    Module.Index0Float2Write=true
    Module.Index0Float3Write=true
   )
   /SetVariables(Particles.GlowColor=&Multiply_LinearColorByFloat(
     Linear Color=&ColorFromCurve(ColorCurve=DataInterface() {propertyValues {
        AlphaCurve {keys [<at 0.3694 0 RCIM_Cubic>,<at 0.9761 0.6 RCIM_Cubic>]}
        BlueCurve {keys [<at 0.7874 0.2865 RCIM_Cubic>,<at 0.979 0.4635>]}
        GreenCurve {keys [<at 0.7874 0.7543 RCIM_Cubic>,<at 0.979 0.8153>]}
        RedCurve {keys [<at 0.7874 1 RCIM_Cubic>,<at 0.979 1>]}
       }})
     Float=&FNTimeOfDayScalar(Night Scale=0.65, Morning Scale=0.9, Evening Scale=0.85)
    )) {isSetParameters=true}
  ]
  rendererClasses [&NiagaraSpriteRendererProperties]
  renderers [
   {rendererClass=&NiagaraSpriteRendererProperties}
   {rendererClass=&NiagaraSpriteRendererProperties, rendererIndex=1}
  ]
 }
 LWE_AirGlow {rendererClasses [], renderers []}
 LWE_Rays {rendererClasses [], renderers []}
 SmallHearts {
  emitterUpdateScript [
   &EmitterState()
   &SpawnBurst_Instantaneous(
    Spawn Count=11
    Spawn Time=&Subtract_Float(A=$System.IntroLength, B=0.07)
    Use Spawn Probability=false
    UseLoopCountLimit=false
   )
  ]
  particleSpawnScript [
   /SetVariables(Particles.HeartPicker:NiagaraFloat=&UniformAOrBFloat(A=2, B=3, Evaluation Type=::NewEnumerator0, Recalculate Random Each Loop=true, Override Seed=false)) {isSetParameters=true}
   &InitializeParticle(
    Color=&LerpLinearColors(
     StartColor=&LerpLinearColors(StartColor=<rgba 0.4896 0.3121 0.141 1>, EndColor=<rgba 0.4896 0.3121 0.141 1>, LerpFactor=$Particles.HeartPicker)
     EndColor=&LerpLinearColors(StartColor=<rgba 0.0625 0.0342 0.0163 1>, EndColor=<rgba 0.0521 0.0285 0.0136 1>, LerpFactor=$Particles.HeartPicker)
     LerpFactor=&UniformAOrBFloat(A=0, Evaluation Type=::NewEnumerator0, Recalculate Random Each Loop=true, Override Seed=false)
    )
    Lifetime=&RandomRangeFloat(Minimum=0.5, Maximum=0.7)
    Position Offset=$System.BurstPosition
    Sprite Size Mode=::NewEnumerator3
    Uniform Sprite Size=&RandomRangeFloat(Minimum=12, Maximum=18)
    UsePositionOffset=true
    Write Material Random=true
   )
   /SetVariables(Particles.SubImageIndex=$Particles.HeartPicker) {isSetParameters=true}
   &DynamicMaterialParameters(
    Index 0 Param 3=1
    Module.Index0Float0Write=true
    Module.Index0Float1Write=true
    Module.Index0Float2Write=true
    Module.Index0Float3Write=true
   )
   &ShapeLocation(
    MaskSpawnGroup=false
    Offset=<0 0 -5.6656>
    Offset Mode=::NewEnumerator0
    Override Seed=false
    Sphere Radius=11
    Sphere Surface Distribution=1
    UseOwnerScale=true
   )
   &AddVelocity(Velocity=&Add_Vector(
     A=&Multiply_Vector(
      A=&NormalizeVector(
       Vector To Normalize=&Subtract_Vector(A=Linked() {linkedVariable {name=Particles.Position, type=NiagaraPosition}}, B=$System.BurstPosition)
       Safe Normalize=true
       Safe Fallback Threshold=0
       Safe Fallback Vector=<0 0 0.001>
      )
      B=&RandomRangeVector(Minimum=<35 35 0>, Maximum=<50 50 0>)
     )
     B=&RandomRangeVector(Minimum=<0 0 40>, Maximum=<0 0 80>)
    ))
   &ApplyInitialForces()
  ]
  particleUpdateScript [
   &ParticleState()
   &ScaleSpriteSize(
    Scale Sprite Size Mode=::NewEnumerator0
    Uniform Scale Factor=&FloatFromCurve(FloatCurve=DataInterface() {propertyValues {Curve {
        defaultValue=0
        keys [
         {arriveTangent=18.0926, interpMode=RCIM_Cubic, leaveTangent=18.0926, tangentMode=RCTM_User, time=0, value=0}
         {arriveTangent=0.4904, interpMode=RCIM_Cubic, leaveTangent=0.4904, time=0.1088, value=1.073}
         {arriveTangent=-0.0405, interpMode=RCIM_Cubic, leaveTangent=-0.0405, tangentMode=RCTM_User, time=1, value=0.4904}
        ]
       }}})
   ) {module=ScaleSpriteSize002}
   &ScaleColor(
    Scale Alpha=&FloatFromCurve(
     FloatCurve=DataInterface() {propertyValues {Curve {
        defaultValue=0
        keys [
         {arriveTangent=-0.2441, time=0.05, value=1}
         {arriveTangent=-1.0526, interpMode=RCIM_Cubic, leaveTangent=-1.0526, time=0.3022, value=0.579}
         {arriveTangent=-0.0405, interpMode=RCIM_Cubic, leaveTangent=-0.0405, tangentMode=RCTM_User, time=1, value=0}
        ]
       }}}
     Scale Curve=&RandomRangeFloat(Minimum=0.8, Maximum=1.5)
    )
    ScaleA=true
    ScaleRGB=true
   )
   &GravityForce(Gravity=&Multiply_VectorByFloat(
     Vector=&RandomRangeVector(Minimum=<0 0 -100>, Maximum=<0 0 -200>)
     Float=&FloatFromCurve(FloatCurve=DataInterface() {propertyValues {Curve {keys [{arriveTangent=1.4708, interpMode=RCIM_Cubic, leaveTangent=1.4708, tangentMode=RCTM_User, time=0, value=0.2148}, <at 1 1>]}}})
    ))
   &SolveForcesAndVelocity(Clamp Velocity=false, Limit Acceleration=false)
  ]
  rendererClasses [&NiagaraSpriteRendererProperties]
  renderers [{rendererClass=&NiagaraSpriteRendererProperties}]
 }
}
summary {
 systemName=NS_ReactFX_PromDutch
 emitters [
  {emitterName=Intro, rendererClasses [&NiagaraSpriteRendererProperties]}
  {emitterName=SmallHearts, rendererClasses [&NiagaraSpriteRendererProperties]}
  {emitterName=LWE_AirGlow, rendererClasses []}
  {emitterName=LWE_Rays, rendererClasses []}
 ]
 userVariables [
  User.bIsFrontEnd {type=NiagaraBool}
  User.bIsFrontEndPreview {type=NiagaraBool}
 ]
}
