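# Niagara system for the QuiltLaze cosmetic companion's reaction effect: a heart intro that pops into a burst, with a petting-reaction sound.
# intro_no_pulse plays the audio and moves a spinning, growing, time-of-day-tinted heart toward System.BurstPosition over System.IntroLength.
# Just before that ends, Pop001 flashes and Hearts throws out shrinking, flickering sprites under gravity and drag.
# Spawn groups pick per-variant values via MergeHelper switches and set VisibilityTag across multiple sprite renderers.
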NiagaraSystem /CosmeticCompanions/FX/React/NS_ReactFX_QuiltLaze
dependencies {
 usedDataInterfaces [
  &NiagaraDataInterfaceCurve
  &NiagaraDataInterfaceColorCurve
  &NiagaraDataInterfaceVectorCurve
  &NiagaraDataInterfaceAudioPlayer
 ]
 usedDynamicInputs [/Niagara/DynamicInputs/TypeConversions/MakeCustomPositionFromBool, &FN_FrontEndVector3Modifier]
 usedModules [
  &self:SystemSpawnScript.SystemScriptSource.SystemScriptGraph.NiagaraNodeAssignment_0.SetVariables_#1
  &self:SystemSpawnScript.SystemScriptSource.SystemScriptGraph.NiagaraNodeAssignment_1.SetVariables_#2
  &self:SystemSpawnScript.SystemScriptSource.SystemScriptGraph.NiagaraNodeAssignment_8.SetVariables_#3
  &self:SystemSpawnScript.SystemScriptSource.SystemScriptGraph.NiagaraNodeAssignment_4.SetVariables_#4
  &self:SystemSpawnScript.SystemScriptSource.SystemScriptGraph.NiagaraNodeAssignment_9.SetVariables_#5
  &self:SystemSpawnScript.SystemScriptSource.SystemScriptGraph.NiagaraNodeAssignment_5.SetVariables_#6
  &self:SystemSpawnScript.SystemScriptSource.SystemScriptGraph.NiagaraNodeAssignment_6.SetVariables_#7
  &self:SystemSpawnScript.SystemScriptSource.SystemScriptGraph.NiagaraNodeAssignment_7.SetVariables_#8
  &self:SystemSpawnScript.SystemScriptSource.SystemScriptGraph.NiagaraNodeAssignment_3.SetVariables_#9
  &self:SystemSpawnScript.SystemScriptSource.SystemScriptGraph.NiagaraNodeAssignment_2.SetVariables_#10
  &SystemState
 ]
}
emitters {
 Hearts {
  emitterUpdateScript [
   &EmitterState()
   &SpawnBurst_Instantaneous(Spawn Count=7, Spawn Group=3, Spawn Time=0.75, Use Spawn Probability=false, UseLoopCountLimit=false) {enabled=false, module=SpawnBurst_Instantaneous003}
   &SpawnBurst_Instantaneous(Spawn Count=7, Spawn Group=2, Spawn Time=0.75, Use Spawn Probability=false, UseLoopCountLimit=false) {enabled=false, module=SpawnBurst_Instantaneous002}
   &SpawnBurst_Instantaneous(
    Spawn Count=10
    Spawn Time=&Subtract_Float(A=$System.IntroLength, B=0.04)
    Use Spawn Probability=false
    UseLoopCountLimit=false
   )
   &SpawnBurst_Instantaneous(
    Spawn Count=&RandomRangeInt(Minimum=11, Maximum=13)
    Spawn Group=1
    Spawn Time=&Subtract_Float(A=$System.IntroLength, B=0.04)
    Use Spawn Probability=false
    UseLoopCountLimit=false
   ) {module=SpawnBurst_Instantaneous001}
  ]
  particleSpawnScript [
   /SetVariables(Particles.VisibilityTag=&SpawnGroup()) {isSetParameters=true}
   &InitializeParticle(
    Color=&MergeHelper_ColorSwitch(
     0=&RandomRangeLinearColor(Minimum=<rgba 1 0.0489 0.2023 1>, Maximum=<rgba 0.6038 0.0296 0.1221 1>)
     1=&RandomRangeLinearColor(Minimum=<rgba 5 4.28 0.0633 1>, Maximum=<rgba 0.4031 0 0 1>)
     2=&RandomRangeLinearColor(Minimum=<rgba 1 0.0489 0.2023 1>, Maximum=<rgba 0.6038 0.0296 0.1221 1>)
     3=&RandomRangeLinearColor(Minimum=<rgba 5 4.28 0.0633 1>, Maximum=<rgba 0.4031 0 0 1>)
     Index=&SpawnGroup()
     Num=::NewEnumerator2
    )
    Lifetime=&MergeHelper_FloatSwitch(
     0=&RandomRangeFloat(Minimum=0.4, Maximum=0.6)
     1=&RandomRangeFloat(Minimum=0.3, Maximum=0.5)
     2=&RandomRangeFloat(Minimum=0.3, Maximum=0.5)
     3=&RandomRangeFloat(Minimum=0.4, Maximum=0.6)
     Index=&SpawnGroup()
     Num=::NewEnumerator2
    )
    Position Offset=$System.BurstPosition
    Sprite Rotation Mode=::NewEnumerator2
    Sprite Size Mode=::NewEnumerator3
    Uniform Sprite Size=&MergeHelper_FloatSwitch(
     0=&Multiply_Float(A=1, B=&RandomRangeFloat(Minimum=4, Maximum=16))
     1=&RandomRangeFloat(Minimum=8, Maximum=22)
     2=&RandomRangeFloat(Minimum=8, Maximum=22)
     3=&RandomRangeFloat(Minimum=8, Maximum=15)
     Index=&SpawnGroup()
     Num=::NewEnumerator2
    )
    UsePositionOffset=true
    Write Material Random=true
   )
   &ShapeLocation(
    MaskSpawnGroup=false
    Sphere Distribution=::NewEnumerator3
    Sphere Radius=&MergeHelper_FloatSwitch(0=10, 1=10, 2=10, 3=10, Index=&SpawnGroup(), Num=::NewEnumerator2)
    UseOwnerScale=true
   )
   &AddVelocity(Velocity=&MergeHelper_VectorSwitch(
     0=&RandomRangeVector(Minimum=<-70 -70 50>, Maximum=<70 70 300>)
     1=&RandomRangeVector(Minimum=<-100 -100 100>, Maximum=<100 100 300>)
     2=?
     3=?
     Index=&SpawnGroup()
     Num=::NewEnumerator0
    )) {enabled=false, module=AddVelocity001}
   &AddVelocity(
    Constrain To Radius=100
    Module.Use Falloff Distance=false
    Origin Offset=<0 0 1>
    Override Seed=false
    Velocity Mode=::NewEnumerator1
    Velocity Origin:Vector3f=$System.BurstPosition
    Velocity Speed=&MergeHelper_FloatSwitch(0=150, 1=150, 2=100, 3=150, Index=&SpawnGroup(), Num=::NewEnumerator2)
   )
  ]
  particleUpdateScript [
   &ParticleState()
   &SpriteRotationRate(Rotation Rate=&MergeHelper_FloatSwitch(0=0, 1=0, 2=180, 3=0, 4=0, Index=&SpawnGroup(), Num=::NewEnumerator3))
   &ScaleSpriteSize(Uniform Curve Sprite Scale=DataInterface() {propertyValues {Curve {
       defaultValue=1
       keys [
        {arriveTangent=4.1399, interpMode=RCIM_Cubic, leaveTangent=4.1399, tangentMode=RCTM_User, time=0, value=0.5}
        {arriveTangent=0.0224, interpMode=RCIM_Cubic, leaveTangent=0.0224, tangentMode=RCTM_User, time=1, value=1}
       ]
      }}})
   /SetVariables(Particles.Flicker:NiagaraFloat=&Waveform(
     "Waveform [1]"=::NewEnumerator0
     "[1] Frequency"=&RandomRangeFloat(Minimum=3, Maximum=5)
     "[1] Amplitude Min/Max"=<0 1>
    )) {isSetParameters=true}
   &GravityForce(Gravity=&MergeHelper_VectorSwitch(
     0=&RandomRangeVector(Minimum=<0 0 -150>, Maximum=<0 0 -500>)
     1=&RandomRangeVector(Minimum=<0 0 -150>, Maximum=<0 0 -500>)
     2=&RandomRangeVector(Minimum=<0 0 -600>, Maximum=<0 0 -800>)
     3=&RandomRangeVector(Minimum=<0 0 -600>, Maximum=<0 0 -800>)
     Index=&SpawnGroup()
     Num=::NewEnumerator2
    ))
   &Drag(
    Drag=&MergeHelper_FloatSwitch(
     0=&Multiply_Float(
      A=&RandomRangeFloat(Minimum=3, Maximum=7)
      B=&FloatFromCurve(
       FloatCurve=DataInterface() {propertyValues {Curve {defaultValue=1, keys [{tangentMode=RCTM_User, time=0.2, value=0}, {tangentMode=RCTM_User, time=0.35, value=2}]}}}
       Scale Curve=3
      )
     )
     1=&Multiply_Float(
      A=&RandomRangeFloat(Minimum=1, Maximum=5)
      B=&FloatFromCurve(FloatCurve=DataInterface() {propertyValues {Curve {defaultValue=1, keys [{tangentMode=RCTM_User, time=0.2, value=0}, {tangentMode=RCTM_User, time=0.35, value=2}]}}})
     )
     2=5
     3=5
     Index=&SpawnGroup()
     Num=::NewEnumerator2
    )
    Use Linear Drag=true
    Use Rotational Drag=true
   )
   &CurlNoiseForce(
    Bias Noise=false
    Noise Frequency=200
    Noise Strength=&RandomRangeFloat(Minimum=200, Maximum=400)
    Pan Noise=false
   ) {enabled=false}
   &PointAttractionForce(
    Attraction Radius=600
    Attraction Strength=&FloatFromCurve(
     FloatCurve=DataInterface() {propertyValues {Curve {
        defaultValue=0
        keys [{interpMode=RCIM_Cubic, tangentMode=RCTM_User, time=0, value=0}, {interpMode=RCIM_Cubic, tangentMode=RCTM_User, time=0.4, value=1}]
       }}}
     Scale Curve=500
    )
    Attractor Position Offset:Vector3f=<0 0 -500>
    Kill Within Radius=false
    Module.Use Falloff=true
    UseAttractorOffset=true
   ) {enabled=false}
   &SolveForcesAndVelocity(Clamp Velocity=false, Limit Acceleration=false)
   &ScaleColor(
    Scale Alpha=&MergeHelper_FloatSwitch(
     0=&FloatFromCurve(FloatCurve=DataInterface() {propertyValues {Curve {
         defaultValue=1
         keys [
          {arriveTangent=-0.2867, interpMode=RCIM_Cubic, leaveTangent=-0.2867, tangentMode=RCTM_User, time=0, value=0}
          {arriveTangent=-3.8875, interpMode=RCIM_Cubic, leaveTangent=-3.8875, tangentMode=RCTM_User, time=0.3695, value=1}
          {arriveTangent=-0.0405, interpMode=RCIM_Cubic, leaveTangent=-0.0405, tangentMode=RCTM_User, time=0.8, value=-0.0266}
         ]
        }}})
     1=&Multiply_Float(
      A=&FloatFromCurve(FloatCurve=DataInterface() {propertyValues {Curve {
          defaultValue=1
          keys [
           {arriveTangent=27.2232, interpMode=RCIM_Cubic, leaveTangent=27.2232, tangentMode=RCTM_User, time=0, value=0}
           {arriveTangent=0.4162, interpMode=RCIM_Cubic, leaveTangent=0.4162, tangentMode=RCTM_User, time=0.0689, value=1.2342}
           {arriveTangent=-0.0405, interpMode=RCIM_Cubic, leaveTangent=-0.0405, tangentMode=RCTM_User, time=1, value=0}
          ]
         }}})
      B=$Particles.Flicker
     )
     2=1
     3=?
     Index=&SpawnGroup()
     Num=::NewEnumerator1
    )
    Scale RGB=&MergeHelper_VectorSwitch(
     0=<1 1 1>
     1=&MakeVectorFromLinearColorRGB(Color=&ColorFromCurve(ColorCurve=DataInterface() {propertyValues {
         AlphaCurve {keys []}
         BlueCurve {keys [<at 0.0637 20>,<at 0.1009 1>,<at 0.3315 1 RCIM_Cubic>,<at 0.3329 1>,<at 0.9944 0.1221>]}
         GreenCurve {keys [<at 0.0637 20>,<at 0.1009 1>,<at 0.3315 1 RCIM_Cubic>,<at 0.3329 1>,<at 0.9944 0.0296>]}
         RedCurve {keys [<at 0.0637 20>,<at 0.1009 1>,<at 0.3315 1 RCIM_Cubic>,<at 0.3329 1>,<at 0.9944 0.6038>]}
        }}))
     2=<1 1 1>
     3=?
     Index=&SpawnGroup()
     Num=::NewEnumerator1
    )
    ScaleA=true
    ScaleRGB=true
   )
   /SetVariables(
    Particles.Dots Spritesize:Vector2f=&Vector2DFromFloat(Value=&FloatFromCurve(
      FloatCurve=DataInterface() {propertyValues {Curve {
         defaultValue=1
         keys [
          {arriveTangent=2.7312, interpMode=RCIM_Cubic, leaveTangent=2.7312, tangentMode=RCTM_User, time=0, value=0.5}
          {arriveTangent=0.0224, interpMode=RCIM_Cubic, leaveTangent=0.0224, tangentMode=RCTM_User, time=1, value=1}
         ]
        }}}
      Scale Curve=&Multiply_Float(A=1, B=&RandomRangeFloat(Minimum=2, Maximum=7))
     ))
    Particles.Hearts SpriteSize:Vector2f=&Vector2DFromFloat(Value=&FloatFromCurve(
      FloatCurve=DataInterface() {propertyValues {Curve {
         defaultValue=1
         keys [
          {arriveTangent=3.2566, interpMode=RCIM_Cubic, leaveTangent=3.2566, tangentMode=RCTM_User, time=0, value=0.5}
          {arriveTangent=-0.5, interpMode=RCIM_Cubic, leaveTangent=-0.5, time=0.5039, value=1.1092}
          {arriveTangent=0.0224, interpMode=RCIM_Cubic, leaveTangent=0.0224, tangentMode=RCTM_User, time=1, value=0}
         ]
        }}}
      Scale Curve=&Multiply_Float(A=1, B=&RandomRangeFloat(Minimum=4, Maximum=16))
     ))
    Particles.Small hearts SpriteSize:Vector2f=&Vector2DFromFloat(Value=&FloatFromCurve(
      FloatCurve=DataInterface() {propertyValues {Curve {
         defaultValue=1
         keys [
          {arriveTangent=1.914, interpMode=RCIM_Cubic, leaveTangent=1.914, tangentMode=RCTM_User, time=0, value=0.5}
          {arriveTangent=-0.1241, interpMode=RCIM_Cubic, leaveTangent=-0.1241, tangentMode=RCTM_User, time=1, value=1}
         ]
        }}}
      Scale Curve=&Multiply_Float(A=1, B=&RandomRangeFloat(Minimum=3, Maximum=8))
     ))
    Particles.Sparkles SpriteSize:Vector2f=&Vector2DFromFloat(Value=&FloatFromCurve(
      FloatCurve=DataInterface() {propertyValues {Curve {keys [<at 0 1>,<at 1 0>]}}}
      Scale Curve=&Multiply_Float(A=1, B=&RandomRangeFloat(Minimum=8, Maximum=22))
     ))
   ) {isSetParameters=true}
  ]
  rendererClasses [&NiagaraSpriteRendererProperties]
  renderers [
   {rendererClass=&NiagaraSpriteRendererProperties}
   {rendererClass=&NiagaraSpriteRendererProperties, rendererIndex=1}
   {rendererClass=&NiagaraSpriteRendererProperties, rendererIndex=2}
   {rendererClass=&NiagaraSpriteRendererProperties, rendererIndex=3}
  ]
 }
 Pop001 {
  emitterUpdateScript [
   &EmitterState()
   &SpawnBurst_Instantaneous(
    Spawn Group=1
    Spawn Time=&Subtract_Float(A=$System.IntroLength, B=0.05)
    Use Spawn Probability=false
    UseLoopCountLimit=false
   ) {module=SpawnBurst_Instantaneous001}
   &SpawnBurst_Instantaneous(
    Spawn Time=&Subtract_Float(A=$System.IntroLength, B=0.05)
    Use Spawn Probability=false
    UseLoopCountLimit=false
   )
  ]
  particleSpawnScript [
   /SetVariables(Particles.VisibilityTag=&SpawnGroup()) {isSetParameters=true}
   &InitializeParticle(
    Lifetime=0.2
    Position Offset=$System.BurstPosition
    Sprite Rotation Mode=::NewEnumerator0
    Sprite Size Mode=::NewEnumerator3
    Uniform Sprite Size=&Multiply_Float(A=12, B=$System.Heart Size)
    UsePositionOffset=true
    Write Material Random=true
   )
   &DynamicMaterialParameters(
    Index 0 Param 4=1
    Module.Index0Float0Write=true
    Module.Index0Float1Write=true
    Module.Index0Float2Write=true
    Module.Index0Float3Write=true
   )
   /SetVariables(Particles.center Linear Color:LinearColor=<rgba 1 0.8431 0 1>, Particles.center spike SpriteRotation=0.0234) {isSetParameters=true}
  ]
  particleUpdateScript [
   &ParticleState()
   &ScaleSpriteSize(
    Scale Sprite Size Mode=::NewEnumerator0
    Uniform Scale Factor=&FloatFromCurve(
     FloatCurve=DataInterface() {propertyValues {Curve {keys [
         {arriveTangent=3.3458, interpMode=RCIM_Cubic, leaveTangent=3.3458, tangentMode=RCTM_User, time=0, value=1}
         {arriveTangent=1.251, tangentMode=RCTM_User, time=0.45, value=3}
        ]}}}
     Scale Curve=2
    )
   ) {module=ScaleSpriteSize002}
   &ScaleColor(
    Scale Alpha=&FloatFromCurve(FloatCurve=DataInterface() {propertyValues {Curve {
        defaultValue=0
        keys [
         {arriveTangent=0.3137, interpMode=RCIM_Cubic, leaveTangent=0.3137, tangentMode=RCTM_User, time=0.2, value=1}
         {arriveTangent=-2.0958, interpMode=RCIM_Cubic, leaveTangent=-2.0958, tangentMode=RCTM_User, time=1, value=0}
        ]
       }}})
    Scale RGB=&VectorFromCurve(VectorCurve=DataInterface() {propertyValues {
       XCurve {keys [{interpMode=RCIM_Cubic, tangentMode=RCTM_User, time=0, value=1}, {arriveTangent=-0.2294, tangentMode=RCTM_User, time=1, value=0}]}
       YCurve {keys [{interpMode=RCIM_Cubic, tangentMode=RCTM_User, time=0, value=1}, {arriveTangent=-0.2294, tangentMode=RCTM_User, time=1, value=0}]}
       ZCurve {keys [{interpMode=RCIM_Cubic, tangentMode=RCTM_User, time=0, value=1}, {arriveTangent=-0.2294, tangentMode=RCTM_User, time=1, value=0}]}
      }})
    ScaleA=true
    ScaleRGB=true
   )
   &SolveForcesAndVelocity(Clamp Velocity=false, Limit Acceleration=false)
   /SetVariables(
    Particles.Glow Color=&Multiply_LinearColorByFloat(
     Linear Color=&MultiplyLinearColor(Linear Color=$Particles.Color, Scale Factor=&MakeVector4_FromLinearColor(LinearColor=<rgba 1 0.8916 0 1>))
     Float=0.33
    )
    Particles.center position:NiagaraPosition=&AddVectorToPosition(Position=Linked() {linkedVariable {name=System.BurstPosition, type=Vector3f}}, Vector=<0 0 -0.7275>)
    Particles.center sprite size:Vector2f=&Vector2DFromFloat(Value=&FloatFromCurve(
      FloatCurve=DataInterface() {propertyValues {Curve {keys [
          {arriveTangent=3.3458, interpMode=RCIM_Cubic, leaveTangent=3.3458, tangentMode=RCTM_User, time=0, value=1}
          {arriveTangent=1.251, tangentMode=RCTM_User, time=0.5, value=3}
         ]}}}
      Scale Curve=6.5
     ))
   ) {isSetParameters=true}
   &DynamicMaterialParameters(
    Index 0 Param 3=&FloatFromCurve(FloatCurve=DataInterface() {propertyValues {Curve {keys [<at 0 1>,<at 1 0>]}}}, Scale Curve=2)
    Index 0 Param 4=&FloatFromCurve(FloatCurve=DataInterface() {propertyValues {Curve {keys [<at 0 1>,<at 0 1>,<at 0.9068 0.9891>,<at 1 0>]}}}, Scale Curve=5)
    Module.Index0Float0Write=true
    Module.Index0Float1Write=true
    Module.Index0Float2Write=false
    Module.Index0Float3Write=true
   ) {module=DynamicMaterialParameters001}
  ]
  rendererClasses [&NiagaraSpriteRendererProperties]
  renderers [
   {rendererClass=&NiagaraSpriteRendererProperties}
   {rendererClass=&NiagaraSpriteRendererProperties, rendererIndex=1}
   {rendererClass=&NiagaraSpriteRendererProperties, rendererIndex=2}
  ]
 }
 intro_no_pulse {
  emitterSpawnScript [
   &InitializeAudioPlayer(Audio Player=DataInterface() {propertyValues {
      Attenuation=null
      Concurrency=null
      MaxPlaysPerTick=10
      SoundToPlay=/CosmeticCompanions/Assets/Audio_Shared/Interaction/Petting/SC_Mimosa_Interaction_React_Default
      bAllowLoopingOneShotSounds=false
      bLimitPlaysPerTick=true
      bOnlyActiveDuringGameplay=false
      bStopWhenComponentIsDestroyed=true
      ConfigurationUserParameter {parameter {name=None, typeDefHandle {registeredTypeIndex=1}, varData []}}
      ParameterNames []
     }})
   /Niagara/Modules/Audio/PlayAudio() {module=PlayAudio001}
  ]
  emitterUpdateScript [
   &EmitterState()
   &SpawnBurst_Instantaneous(Spawn Group=1, Use Spawn Probability=false, UseLoopCountLimit=false) {module=SpawnBurst_Instantaneous001}
   &SpawnBurst_Instantaneous(Use Spawn Probability=false, UseLoopCountLimit=false)
  ]
  particleSpawnScript [
   /SetVariables(Particles.VisibilityTag=&SpawnGroup()) {isSetParameters=true}
   &InitializeParticle(
    Lifetime=$System.IntroLength
    Sprite Rotation Mode=::NewEnumerator0
    Sprite Size Mode=::NewEnumerator3
    Uniform Sprite Size=&MergeHelper_FloatSwitch(
     0=&Multiply_Float(A=8, B=$System.Heart Size)
     1=&Multiply_Float(A=8, B=$System.Heart Size)
     2=?
     3=?
     Index=&SpawnGroup()
     Num=::NewEnumerator0
    )
    UsePositionOffset=false
    Write Material Random=true
   )
  ]
  particleUpdateScript [
   &ParticleState()
   &ScaleSpriteSize(
    Scale Sprite Size Mode=::NewEnumerator0
    Uniform Scale Factor=&Add_Float(
     A=0.5
     B=&FloatFromCurve(FloatCurve=DataInterface() {propertyValues {Curve {
         defaultValue=0
         keys [
          {arriveTangent=0.8193, interpMode=RCIM_Cubic, leaveTangent=0.8193, tangentMode=RCTM_User, time=0, value=0}
          {arriveTangent=-0.1202, interpMode=RCIM_Cubic, leaveTangent=-0.1202, tangentMode=RCTM_User, time=1, value=5}
          <at 1.072 7.0969 RCIM_Cubic>
         ]
        }}})
    )
   ) {module=ScaleSpriteSize002}
   &ScaleColor(
    Scale Alpha=&FloatFromCurve(
     FloatCurve=DataInterface() {propertyValues {Curve {
        defaultValue=0
        keys [
         {arriveTangent=9.0725, interpMode=RCIM_Cubic, leaveTangent=9.0725, tangentMode=RCTM_User, time=0, value=0}
         {arriveTangent=-0.2441, leaveTangent=-0.2441, time=0.05, value=1}
         {arriveTangent=-1.146, leaveTangent=-1.146, time=0.85, value=0.9906}
         {arriveTangent=-0.0405, interpMode=RCIM_Cubic, leaveTangent=-0.0405, tangentMode=RCTM_User, time=1, value=0}
        ]
       }}}
     Scale Curve=&FNTimeOfDayScalar(Night Scale=1, Morning Scale=1, Evening Scale=1)
    )
    ScaleA=true
    ScaleRGB=true
   ) {enabled=false}
   &OffsetPosition(Position Offset=&Lerp_Vector(
     A=$System.HeartIntroSpawnPosition
     B=$System.BurstPosition
     Alpha=&FloatFromCurve(FloatCurve=DataInterface() {propertyValues {Curve {keys [
          {arriveTangent=0.7913, interpMode=RCIM_Cubic, leaveTangent=0.7913, tangentMode=RCTM_User, time=0, value=0}
          {arriveTangent=-0.1074, tangentMode=RCTM_User, time=0.6, value=1}
         ]}}})
    ))
   &DynamicMaterialParameters(
    Index 0 Param 1=1.0117
    Index 1 Param 1=1
    Index 1 Param 2=5
    Index 1 Param 3=10
    Module.Index0Float0Write=true
    Module.Index0Float1Write=true
    Module.Index0Float2Write=true
    Module.Index0Float3Write=true
    Module.Index1Float0Write=true
    Module.Index1Float1Write=false
    Module.Index1Float2Write=false
    Module.Index1Float3Write=false
    Write Parameter Index 1=::NewEnumerator1
   )
   /SetVariables(
    Particles.Glow Color=&Multiply_LinearColorByFloat(
     Linear Color=$System.Layer 1 Color
     Float=&FNTimeOfDayScalar(Night Scale=0.15, Morning Scale=0.1, Day Scale=0.1, Evening Scale=0.1)
    )
    Particles.Glow Size=&Multiply_Vector2D(A=$Particles.SpriteSize, B=<0.6 0.6>)
    Particles.Heart fill color:LinearColor=&MakeLinearColorFromVectorAndFloat(
     "Vector (RGB)"=&MakeVectorFromLinearColorRGB(Color=&FNTimeOfDayColor(
       Night Color=<rgba 0.2239 0.2482 0.3031 1>
       Morning Color=<rgba 0.5198 0.5198 0.5198 1>
       Day Color=<rgba 0.7448 0.7448 0.7448 1>
       Evening Color=<rgba 0.4615 0.4615 0.4615 1>
      ))
     "Float (Alpha)"=&FloatFromCurve(
      FloatCurve=DataInterface() {propertyValues {Curve {
         defaultValue=0
         keys [
          {arriveTangent=9.0725, interpMode=RCIM_Cubic, leaveTangent=9.0725, tangentMode=RCTM_User, time=0, value=0}
          {arriveTangent=-0.2441, leaveTangent=-0.2441, time=0.05, value=1}
          {arriveTangent=-1.146, leaveTangent=-1.146, time=0.85, value=0.9906}
          {arriveTangent=-0.0405, interpMode=RCIM_Cubic, leaveTangent=-0.0405, tangentMode=RCTM_User, time=1, value=0}
         ]
        }}}
      Scale Curve=&FNTimeOfDayScalar(Night Scale=1, Morning Scale=1, Evening Scale=1)
     )
    )
    "Particles.edge color ":LinearColor=&MakeLinearColorFromVectorAndFloat(
     "Vector (RGB)"=<1 1 1>
     "Float (Alpha)"=&FloatFromCurve(
      FloatCurve=DataInterface() {propertyValues {Curve {
         defaultValue=0
         keys [
          {arriveTangent=9.0725, interpMode=RCIM_Cubic, leaveTangent=9.0725, tangentMode=RCTM_User, time=0, value=0}
          {arriveTangent=-0.2441, leaveTangent=-0.2441, time=0.05, value=1}
          {arriveTangent=-1.146, leaveTangent=-1.146, time=0.85, value=0.9906}
          {arriveTangent=-0.0405, interpMode=RCIM_Cubic, leaveTangent=-0.0405, tangentMode=RCTM_User, time=1, value=0}
         ]
        }}}
      Scale Curve=&FNTimeOfDayScalar(Night Scale=1, Morning Scale=1, Evening Scale=1)
     )
    )
   ) {isSetParameters=true}
   &SolveForcesAndVelocity(Clamp Velocity=false, Limit Acceleration=false)
   /SetVariables(Particles.flamePosition:NiagaraPosition=$Particles.Initial.Position) {isSetParameters=true}
   /SetVariables(Particles.SpriteRotation:NiagaraFloat=&FloatFromCurve(FloatCurve=DataInterface() {propertyValues {Curve {keys [{interpMode=RCIM_Cubic, tangentMode=RCTM_Break, time=0, value=-156}, {interpMode=RCIM_Cubic, tangentMode=RCTM_Break, time=0.7, value=0}]}}})) {isSetParameters=true}
   /SetVariables(
    Particles.Heart DynParam:Vector4f=&MakeVector4(
     X=&FloatFromCurve(
      FloatCurve=DataInterface() {propertyValues {Curve {keys [
          {arriveTangent=0.6296, interpMode=RCIM_Cubic, leaveTangent=0.6296, tangentMode=RCTM_User, time=0, value=0}
          <at 0.1026 0.6947>
          <at 0.1855 0.0217>
          <at 0.3091 0.7687>
          <at 0.4994 0.0289>
          <at 0.6956 0.8797>
          <at 0.8584 0.4374>
          <at 1 1>
         ]}}}
      Scale Curve=5
     )
     Y=&FloatFromCurve(
      FloatCurve=DataInterface() {propertyValues {Curve {keys [
          {arriveTangent=0.6296, interpMode=RCIM_Cubic, leaveTangent=0.6296, tangentMode=RCTM_User, time=0, value=0}
          {arriveTangent=1.185, tangentMode=RCTM_User, time=0.2294, value=0.8641}
          <at 1.0498 1>
         ]}}}
      Scale Curve=2.5
     )
     Z=&FloatFromCurve(
      FloatCurve=DataInterface() {propertyValues {Curve {keys [
          {arriveTangent=0.6296, interpMode=RCIM_Cubic, leaveTangent=0.6296, tangentMode=RCTM_User, time=0, value=0}
          {arriveTangent=1.185, tangentMode=RCTM_User, time=0.5085, value=1}
          <at 1.0498 1>
         ]}}}
      Scale Curve=4.3691
     )
     W=0
    )
    Particles.Heart fill DynParam:Vector4f=&MakeVector4(X=4.066, Y=0, Z=0, W=0)
    Particles.center heart DynParam:Vector4f=&MakeVector4(X=0, Y=1, Z=-2, W=0)
   ) {isSetParameters=true}
   /SetVariables(
    Particles.center heart size:Vector2f=&Vector2DFromFloat(Value=&Add_Float(
      A=0.5
      B=&FloatFromCurve(FloatCurve=DataInterface() {propertyValues {Curve {
          defaultValue=0
          keys [
           {arriveTangent=0.8193, interpMode=RCIM_Cubic, leaveTangent=0.8193, tangentMode=RCTM_User, time=0, value=0}
           {arriveTangent=2.055, interpMode=RCIM_Cubic, leaveTangent=2.055, tangentMode=RCTM_User, time=0.1161, value=4.2486}
           {arriveTangent=-0.9053, interpMode=RCIM_Cubic, leaveTangent=-0.9053, tangentMode=RCTM_User, time=0.8783, value=2}
           {arriveTangent=-0.1202, interpMode=RCIM_Cubic, leaveTangent=-0.1202, tangentMode=RCTM_User, time=0.9881, value=5}
           <at 1.0134 0 RCIM_Cubic>
          ]
         }}})
     ))
    Particles.hear edge size:Vector2f=&Vector2DFromFloat(Value=&Add_Float(
      A=0.5
      B=&FloatFromCurve(FloatCurve=DataInterface() {propertyValues {Curve {
          defaultValue=0
          keys [
           {arriveTangent=0.8193, interpMode=RCIM_Cubic, leaveTangent=0.8193, tangentMode=RCTM_User, time=0, value=0}
           {arriveTangent=24.3385, interpMode=RCIM_Cubic, leaveTangent=24.3385, time=0.1356, value=5.4758}
           {interpMode=RCIM_Cubic, leaveTangent=-8.9585, time=1.0057, value=24.4782}
          ]
         }}})
     ))
   ) {enabled=false, isSetParameters=true}
  ]
  rendererClasses [&NiagaraSpriteRendererProperties]
  renderers [
   {rendererClass=&NiagaraSpriteRendererProperties}
   {rendererClass=&NiagaraSpriteRendererProperties, rendererIndex=1}
   {rendererClass=&NiagaraSpriteRendererProperties, rendererIndex=2}
  ]
 }
}
summary {
 systemName=NS_ReactFX_QuiltLaze
 emitters [
  {emitterName=Hearts, rendererClasses [&NiagaraSpriteRendererProperties]}
  {emitterName=Pop001, rendererClasses [&NiagaraSpriteRendererProperties]}
  {emitterName=intro_no_pulse, rendererClasses [&NiagaraSpriteRendererProperties]}
 ]
 userVariables [
  User.bIsFrontEnd {type=NiagaraBool}
  User.bIsFrontEndPreview {type=NiagaraBool}
  User.ParaGlideLeanAlpha {type=NiagaraFloat, defaultValue=0, description=""}
 ]
}
