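# Niagara reaction effect for the RareKrow cosmetic companion: an animated fiery heart that wiggles, then bursts, with a petting-reaction sound.
# Intro_FieryHeart adds halo, smoke and glow, Flame_Sprites adds flames and sparks, and delayed Burst emitters add the shattering heart, embers and spike ring.
# Emitters share System.BurstPosition, using spawn groups to vary behavior and visibility tags within each.
# Front-end flags adjust halo and ring; both SmokeLWE emitters are empty.
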NiagaraSystem /CosmeticCompanions/FX/React/NS_ReactFX_RareKrow
dependencies {
 usedDataInterfaces [
  &NiagaraDataInterfaceAudioPlayer
  &NiagaraDataInterfaceSpriteRendererInfo
  &NiagaraDataInterfaceCurve
  &NiagaraDataInterfaceColorCurve
 ]
 usedDynamicInputs [&MakeFloatFromVector2d, &FN_FrontEndVector3Modifier]
 usedModules [
  &self:SystemSpawnScript.SystemScriptSource.SystemScriptGraph.NiagaraNodeAssignment_0.SetVariables_#1
  &self:SystemSpawnScript.SystemScriptSource.SystemScriptGraph.NiagaraNodeAssignment_1.SetVariables_#2
  &self:SystemSpawnScript.SystemScriptSource.SystemScriptGraph.NiagaraNodeAssignment_4.SetVariables_#3
  &self:SystemSpawnScript.SystemScriptSource.SystemScriptGraph.NiagaraNodeAssignment_2.SetVariables_#4
  &self:SystemSpawnScript.SystemScriptSource.SystemScriptGraph.NiagaraNodeAssignment_3.SetVariables_#5
  &self:SystemSpawnScript.SystemScriptSource.SystemScriptGraph.NiagaraNodeAssignment_9.SetVariables_#6
  &self:SystemSpawnScript.SystemScriptSource.SystemScriptGraph.NiagaraNodeAssignment_10.SetVariables_#7
  &self:SystemSpawnScript.SystemScriptSource.SystemScriptGraph.NiagaraNodeAssignment_5.SetVariables_#8
  &self:SystemSpawnScript.SystemScriptSource.SystemScriptGraph.NiagaraNodeAssignment_6.SetVariables_#9
  &self:SystemSpawnScript.SystemScriptSource.SystemScriptGraph.NiagaraNodeAssignment_7.SetVariables_#10
  &SystemState
 ]
}
emitters {
 Burst {
  emitterUpdateScript [
   &EmitterState()
   &SpawnBurst_Instantaneous(Spawn Count=35, Spawn Time=0.75, Use Spawn Probability=false, UseLoopCountLimit=false)
   &SpawnBurst_Instantaneous(
    Spawn Group=1
    Spawn Time=&Subtract_Float(A=$System.IntroLength, B=-0.03)
    Use Spawn Probability=false
    UseLoopCountLimit=false
   ) {module=SpawnBurst_Instantaneous002}
  ]
  particleSpawnScript [
   &InitializeParticle(
    Color=&RandomRangeLinearColor(Minimum=<rgba 2 1.0034 0.0598 0.97>, Maximum=<rgba 2 0.353 0 1>)
    Lifetime=&MergeHelper_FloatSwitch(
     0=&RandomRangeFloat(Minimum=0.5, Maximum=1)
     1=0.3
     2=?
     3=?
     Index=&SpawnGroup()
     Num=::NewEnumerator0
    )
    Position Offset=&AddVectorToPosition(Position=Linked() {linkedVariable {name=System.BurstPosition, type=Vector3f}}, Vector=<0 0 -5>)
    Sprite Size Max=<2.2 8>
    Sprite Size Min=<1.5 4>
    Sprite Size Mode=::NewEnumerator2
    UsePositionOffset=true
    Write Material Random=true
   )
   &AddVelocity(
    Cone Angle=180
    Cone Axis=<0 0 1>
    Module.UseConeFalloff=false
    Module.UseVelDistribution=false
    Override Seed=false
    Velocity Mode=::NewEnumerator2
    Velocity Speed=&MergeHelper_FloatSwitch(
     0=&RandomRangeFloat(Minimum=50, Maximum=80)
     1=0
     2=?
     3=?
     Index=&SpawnGroup()
     Num=::NewEnumerator0
    )
   )
   &ShapeLocation(
    MaskSpawnGroup=false
    Offset=&MergeHelper_VectorSwitch(0=<0 0 -6>, 1=<0 0 0>, 2=?, 3=?, Index=&SpawnGroup(), Num=::NewEnumerator0)
    Offset Mode=::NewEnumerator0
    Override Seed=false
    Shape Origin:Vector3f=$System.BurstPosition
    Sphere Radius=&MergeHelper_FloatSwitch(0=3, 1=0, 2=?, 3=?, Index=&SpawnGroup(), Num=::NewEnumerator0)
    UseOwnerScale=true
   )
   /SetVariables(Particles.VisibilityTag=&SpawnGroup()) {isSetParameters=true}
  ]
  particleUpdateScript [
   &ParticleState()
   &ScaleSpriteSize(Uniform Curve Sprite Scale=DataInterface() {propertyValues {Curve {keys [<at 0 1.5 RCIM_Cubic>,<at 1 0 RCIM_Cubic>]}}})
   &Drag(Drag=&RandomRangeFloat(Minimum=6, Maximum=4), Use Linear Drag=true, Use Rotational Drag=false)
   &CurlNoiseForce(
    Bias Noise=false
    Noise Frequency=100
    Noise Strength=&MergeHelper_FloatSwitch(
     0=&RandomRangeFloat(Minimum=-1000, Maximum=1000)
     1=0
     2=?
     3=?
     Index=&SpawnGroup()
     Num=::NewEnumerator0
    )
    Pan Noise=false
   )
   &GravityForce(
    Coordinate Space=::Local
    Gravity=&Multiply_VectorByFloat(
     Vector=&RandomRangeVector(Minimum=<-100 -100 100>, Maximum=<100 100 200>)
     Float=&MergeHelper_FloatSwitch(0=1, 1=0, 2=?, 3=?, Index=&SpawnGroup(), Num=::NewEnumerator0)
    )
   )
   &SolveForcesAndVelocity(Clamp Velocity=false, Limit Acceleration=false)
   &ScaleSpriteSizeBySpeed(Max Scale Factor=<1 5>)
   &ScaleColor(
    Linear Color Curve=DataInterface() {propertyValues {
      AlphaCurve {keys [<at 0.1413 0>,<at 0.1643 1>,<at 0.3308 1>,<at 1 0>]}
      BlueCurve {keys [<at 0 1>]}
      GreenCurve {keys [<at 0 1>]}
      RedCurve {keys [<at 0 1>]}
     }}
    Scale Mode=2
   )
   /SetVariables(
    Particles.Spike Ring Color:LinearColor=&MakeLinearColorFromVectorAndFloat(
     "Vector (RGB)"=&MakeVectorFromLinearColorRGB(Color=<rgba 3.5 1.771 0 1>)
     "Float (Alpha)"=&FloatFromCurve(
      FloatCurve=DataInterface() {propertyValues {Curve {
         defaultValue=1
         keys [
          {tangentMode=RCTM_User, time=0, value=1}
          {arriveTangent=-0.0405, interpMode=RCIM_Cubic, leaveTangent=-0.0405, tangentMode=RCTM_User, time=0.7009, value=1}
          {arriveTangent=-8.0089, interpMode=RCIM_Cubic, leaveTangent=-8.0089, tangentMode=RCTM_User, time=1, value=0}
         ]
        }}}
      Scale Curve=0.5
     )
    )
    Particles.Spike Ring Size:Vector3f=&Multiply_VectorByFloat(
     Vector=&Multiply_VectorByFloat(
      Vector=<0.37 0.37 0.37>
      Float=&FloatFromCurve(
       FloatCurve=DataInterface() {propertyValues {Curve {
          defaultValue=0
          keys [
           {arriveTangent=18.4197, interpMode=RCIM_Cubic, leaveTangent=18.4197, tangentMode=RCTM_User, time=0, value=0}
           {arriveTangent=1.1179, interpMode=RCIM_Cubic, leaveTangent=1.1179, tangentMode=RCTM_User, time=0.2869, value=0.6072}
           {arriveTangent=0.0224, interpMode=RCIM_Cubic, leaveTangent=0.0224, tangentMode=RCTM_User, time=0.7359, value=0.9376}
           <at 0.9958 1.7804 RCIM_Cubic>
          ]
         }}}
       Scale Curve=0.85
      )
     )
     Float=&MakeCustomFloatFromBool(
      Bool=&Boolean_OrOperation(Bool1=$User.bIsFrontEnd, Bool2=$User.bIsFrontEndPreview)
      True Float=0.85
      False Float=1
     )
    )
   ) {isSetParameters=true}
   /SetVariables(Particles.Ring Dynamic Param:Vector4f=&MakeVector4(
     X=&FloatFromCurve(FloatCurve=DataInterface() {propertyValues {Curve {keys [
          {arriveTangent=-7.2529, interpMode=RCIM_Cubic, leaveTangent=-7.2529, tangentMode=RCTM_User, time=-0.0022, value=1.0993}
          {arriveTangent=0.875, interpMode=RCIM_Cubic, leaveTangent=0.875, tangentMode=RCTM_User, time=0.5, value=0}
          {arriveTangent=3.8908, interpMode=RCIM_Cubic, leaveTangent=3.8908, tangentMode=RCTM_User, time=1, value=1}
         ]}}})
     Y=&FloatFromCurve(FloatCurve=DataInterface() {propertyValues {Curve {keys [
          {arriveTangent=5.2096, interpMode=RCIM_Cubic, leaveTangent=5.2096, tangentMode=RCTM_User, time=0, value=0.5}
          {arriveTangent=0.2132, interpMode=RCIM_Cubic, leaveTangent=0.2132, tangentMode=RCTM_User, time=1, value=3}
         ]}}})
     Z=&FloatFromCurve(FloatCurve=DataInterface() {propertyValues {Curve {
         defaultValue=1
         keys [
          {arriveTangent=0.0994, interpMode=RCIM_Cubic, leaveTangent=0.0994, tangentMode=RCTM_User, time=0, value=1}
          {arriveTangent=-1.6508, interpMode=RCIM_Cubic, leaveTangent=-1.6508, tangentMode=RCTM_User, time=1, value=0}
         ]
        }}})
     W=&FloatFromCurve(
      FloatCurve=DataInterface() {propertyValues {Curve {keys [
          {arriveTangent=2.3563, interpMode=RCIM_Cubic, leaveTangent=2.3563, tangentMode=RCTM_User, time=0, value=0}
          {tangentMode=RCTM_User, time=1, value=1}
         ]}}}
      Scale Curve=2
     )
    )) {isSetParameters=true}
  ]
  rendererClasses [&NiagaraSpriteRendererProperties, &NiagaraMeshRendererProperties]
  renderers [
   {rendererClass=&NiagaraSpriteRendererProperties}
   {rendererClass=&NiagaraMeshRendererProperties, rendererIndex=1}
  ]
 }
 Burst_FieryHeart {
  emitterUpdateScript [
   &EmitterState(Life Cycle Mode=::NewEnumerator1, Loop Behavior=::NewEnumerator1, Loop Delay=0.75, Loop Duration=2, UseLoopDelay=true)
   &SpawnBurst_Instantaneous(Spawn Group=1, Use Spawn Probability=false, UseLoopCountLimit=false) {module=SpawnBurst_Instantaneous001}
   &SpawnBurst_Instantaneous(Use Spawn Probability=false, UseLoopCountLimit=false)
  ]
  particleSpawnScript [
   &InitializeParticle(
    Lifetime=0.3
    Position:Vector3f=$System.BurstPosition
    Position Mode=::NewEnumerator1
    Sprite Size Mode=::NewEnumerator3
    Uniform Sprite Size=35
    UsePositionOffset=true
    Write Material Random=true
   )
   /SetVariables(Particles.VisibilityTag=&SpawnGroup()) {isSetParameters=true, module=SetVariables_#11}
  ]
  particleUpdateScript [
   &ParticleState()
   &SubUVAnimation(
    Direct Frame Index=8
    "Sprite Renderer "=DataInterface() {propertyValues {SpriteRenderer=&self:Burst_FieryHeart_0.NiagaraSpriteRendererProperties_0}}
    SubUV Animation Mode:ENiagara_SubUVLookupModeV2=::NewEnumerator5
   )
   /SetVariables(
    Particles.Linear Color_Halo:LinearColor=&ColorFromCurve(ColorCurve=DataInterface() {propertyValues {
       AlphaCurve {keys [<at 0.9445 0.05 RCIM_Cubic>,<at 1 0 RCIM_Cubic>]}
       BlueCurve {keys [<at 0 1>]}
       GreenCurve {keys [<at 0 1>]}
       RedCurve {keys [<at 0 1>]}
      }})
    Particles.SpriteSize_Halo:Vector2f=&Vector2DFromFloat(Value=37.864)
   ) {isSetParameters=true, module=SetVariables_#12}
   /SetVariables(
    Particles.LinearColor_Electric:LinearColor=&Multiply_LinearColorByFloat(
     Linear Color=&ColorFromCurve(ColorCurve=DataInterface() {propertyValues {
        AlphaCurve {keys [<at 0 1 RCIM_Cubic>,<at 0.9835 0>]}
        BlueCurve {keys [<at 0 1>,<at 0.2927 0>]}
        GreenCurve {keys [<at 0 1>,<at 0.2927 0>]}
        RedCurve {keys [<at 0 1>,<at 0.2927 0>]}
       }})
     Float=1
    )
    Particles.SpriteSize_Electric:Vector2f=&Vector2DFromFloat(Value=27)
   ) {isSetParameters=true, module=SetVariables_#13}
   /SetVariables(Particles.Color_Shatter:LinearColor=&ColorFromCurve(
     ColorCurve=DataInterface() {propertyValues {
       AlphaCurve {keys [<at 0.3053 1>,<at 0.5076 0 RCIM_Cubic>]}
       BlueCurve {keys [<at 0.5115 0 RCIM_Cubic>]}
       GreenCurve {keys [<at 0.5115 0 RCIM_Cubic>]}
       RedCurve {keys [<at 0.5115 0 RCIM_Cubic>]}
      }}
     CurveIndex=&Multiply_Float(A=$Particles.NormalizedAge, B=1)
    )) {isSetParameters=true, module=SetVariables_#14}
   &DynamicMaterialParameters(Write Parameter Index 0=::NewEnumerator0)
  ]
  rendererClasses [&NiagaraSpriteRendererProperties]
  renderers [{rendererClass=&NiagaraSpriteRendererProperties}]
 }
 Flame_Sprites {
  emitterUpdateScript [
   &EmitterState(Life Cycle Mode=::NewEnumerator1, Loop Behavior=::NewEnumerator1, Loop Delay=0.1, Loop Duration=0.5, UseLoopDelay=true)
   &SpawnRate(SpawnGroup=1, SpawnRate=25, Use Spawn Probability=false) {module=SpawnRate001}
   &SpawnRate(
    SpawnRate=&FloatFromCurve(
     FloatCurve=DataInterface() {propertyValues {Curve {keys [{arriveTangent=-2.5918, interpMode=RCIM_Cubic, time=0, value=0}, {interpMode=RCIM_Cubic, leaveTangent=-3.1528, time=0.325, value=1}]}}}
     CurveIndex=$Emitter.NormalizedLoopAge
     Scale Curve=30
    )
    Use Spawn Probability=false
   )
  ]
  particleSpawnScript [
   &InitializeParticle(
    Lifetime=&MergeHelper_FloatSwitch(
     0=&RandomRangeFloat(Minimum=0.125, Maximum=0.325)
     1=&RandomRangeFloat(Minimum=0.2, Maximum=0.4)
     2=?
     3=?
     Index=&SpawnGroup()
     Num=::NewEnumerator0
    )
    Position:Vector3f=$System.BurstPosition
    Position Mode=::NewEnumerator1
    Position Offset=<0 0 10>
    Sprite Rotation Mode=::NewEnumerator2
    Sprite Size Mode=::NewEnumerator3
    Sprite UV Mode=::NewEnumerator3
    Uniform Sprite Size=&MergeHelper_FloatSwitch(
     0=&RandomRangeFloat(Minimum=15, Maximum=22)
     1=&RandomRangeFloat(Minimum=3, Maximum=6)
     2=?
     3=?
     Index=&SpawnGroup()
     Num=::NewEnumerator0
    )
    UsePositionOffset=false
    Write Material Random=true
   )
   &AddVelocity(
    Cone Angle=&MergeHelper_FloatSwitch(0=10, 1=180, 2=?, 3=?, Index=&SpawnGroup(), Num=::NewEnumerator0)
    Cone Axis=&MergeHelper_VectorSwitch(
     0=&MakeVector(
      X=&FloatFromCurve(FloatCurve=DataInterface() {propertyValues {Curve {keys [{arriveTangent=-3.8249, interpMode=RCIM_Cubic, leaveTangent=-3.8249, tangentMode=RCTM_User, time=0.5, value=0.5}, <at 1 -0.25 RCIM_Cubic>]}}})
      Y=0.5
      Z=1
     )
     1=<0 0 1>
     2=?
     3=?
     Index=&SpawnGroup()
     Num=::NewEnumerator0
    )
    Module.UseConeFalloff=false
    Module.UseVelDistribution=false
    Override Seed=false
    Velocity Mode=::NewEnumerator2
    Velocity Speed=&FloatFromCurve(
     FloatCurve=DataInterface() {propertyValues {Curve {keys [{arriveTangent=-7.5919, interpMode=RCIM_Cubic, leaveTangent=-7.5919, tangentMode=RCTM_User, time=0, value=1}, <at 0.2 0.5 RCIM_Cubic>]}}}
     CurveIndex=$Emitter.LoopedAge
     Scale Curve=&MergeHelper_FloatSwitch(
      0=&RandomRangeFloat(Minimum=180, Maximum=300)
      1=&RandomRangeFloat(Minimum=300, Maximum=350)
      2=?
      3=?
      Index=&SpawnGroup()
      Num=::NewEnumerator0
     )
    )
   )
   &CameraOffset(Camera Offset Amount=&RandomRangeFloat(Minimum=-15, Maximum=-20))
   &ShapeLocation(MaskSpawnGroup=false, Non Uniform Scale=<1 1 0.2>, Override Seed=false, Sphere Radius=10, UseOwnerScale=true)
   /SetVariables(Particles.VisibilityTag=&SpawnGroup()) {isSetParameters=true}
   /SetVariables(
    Particles.Sparks Color:LinearColor=&RandomRangeLinearColor(Minimum=<rgba 2 1.0034 0.0598 0.97>, Maximum=<rgba 2 0.353 0 1>)
    Particles.Sparks Sprite Size:Vector2f=&RandomRangeVector2D(Minimum=<1 2.5>, Maximum=<2 4>)
   ) {isSetParameters=true}
  ]
  particleUpdateScript [
   &ParticleState()
   &Drag(
    Drag=&FloatFromCurve(
     FloatCurve=DataInterface() {propertyValues {Curve {keys [{tangentMode=RCTM_User, time=0.2, value=1}, {tangentMode=RCTM_User, time=0.32, value=3}]}}}
     CurveIndex=$Emitter.Age
     Scale Curve=2.5
    )
    Use Linear Drag=true
    Use Rotational Drag=false
   )
   &GravityForce(Gravity=&Multiply_VectorByFloat(
     Vector=&RandomRangeVector(Minimum=<100 -120 300>, Maximum=<150 -250 600>)
     Float=&FloatFromCurve(FloatCurve=DataInterface() {propertyValues {Curve {keys [<at 0 1 RCIM_Cubic>,<at 1 0 RCIM_Cubic>]}}})
    ))
   &CurlNoiseForce(
    Bias Noise=false
    Noise Frequency=150
    Noise Strength=&MergeHelper_FloatSwitch(0=0, 1=500, 2=?, 3=?, Index=&SpawnGroup(), Num=::NewEnumerator0)
    Pan Noise=false
   )
   &SolveForcesAndVelocity(Clamp Velocity=false, Limit Acceleration=false)
   &ScaleSpriteSize(Uniform Curve Sprite Scale=DataInterface() {propertyValues {Curve {keys [
        {arriveTangent=0.5006, interpMode=RCIM_Cubic, leaveTangent=0.5006, leaveTangentWeight=0.7039, tangentMode=RCTM_User, time=0, value=0.725}
        {arriveTangentWeight=0.7039, interpMode=RCIM_Cubic, tangentMode=RCTM_User, time=1, value=1.5}
       ]}}})
   &SpriteRotationRate(Rotation Rate=&RandomRangeFloat(Minimum=0, Maximum=80))
   /Niagara/Modules/Update/SubUV/SubUVAnimation(Number Of Frames=&FloatFromCurve(
     FloatCurve=DataInterface() {propertyValues {Curve {keys [{arriveTangent=4.0889, interpMode=RCIM_Cubic, leaveTangent=4.0889, tangentMode=RCTM_User, time=0, value=56}, <at 0.9 60>]}}}
     Scale Curve=&MaskFloatBySpawnGroup(Masked Float=0.7, Passthrough Float=1.25)
    ))
   &DynamicMaterialParameters(
    Index 0 Param 1=1
    Index 0 Param 2=&RandomRangeFloat(Minimum=0, Maximum=4)
    Index 0 Param 3=&FloatFromCurve(FloatCurve=DataInterface() {propertyValues {Curve {keys [<at 0 0.5 RCIM_Cubic>, {arriveTangent=1.5, interpMode=RCIM_Cubic, leaveTangent=1.5, time=0.5235, value=0.5}, <at 1 2>]}}})
    Index 0 Param 4=2.5
    Module.Index0Param0Write=true
    Module.Index0Param1Write=true
    Module.Index0Param2Write=true
    Module.Index0Param3Write=true
   )
   &ScaleColor(
    Scale Alpha=&Multiply_Float(
     A=&RandomRangeFloat(Minimum=0.6, Maximum=1)
     B=&FloatFromCurve(
      FloatCurve=DataInterface() {propertyValues {Curve {keys [
          {arriveTangent=0.162, interpMode=RCIM_Cubic, leaveTangent=0.162, tangentMode=RCTM_User, time=0, value=0.1}
          {arriveTangent=2.2098, interpMode=RCIM_Cubic, leaveTangent=2.2098, time=0.2, value=0.9}
          {arriveTangent=-0.0116, interpMode=RCIM_Cubic, leaveTangent=-0.0116, tangentMode=RCTM_User, time=0.4073, value=1}
          {arriveTangent=-2.0743, interpMode=RCIM_Cubic, leaveTangent=-2.0743, tangentMode=RCTM_User, time=0.6746, value=0.6742}
          {interpMode=RCIM_Cubic, leaveTangent=-2.371, time=1, value=0}
         ]}}}
      Scale Curve=4
     )
    )
    Scale RGB=&Multiply_Vector(
     A=&MakeVectorFromLinearColorRGB(Color=&ColorFromCurve(ColorCurve=DataInterface() {propertyValues {
         AlphaCurve {keys [<at 0 0.5 RCIM_Cubic>]}
         BlueCurve {keys [<at 0 0 RCIM_Cubic>,<at 0.6107 0>,<at 0.7654 0.0001>]}
         GreenCurve {keys [<at 0 1.7343 RCIM_Cubic>,<at 0.6107 0.2365>,<at 0.7654 0.0172>]}
         RedCurve {keys [<at 0 8 RCIM_Cubic>,<at 0.6107 2>,<at 0.7654 1>]}
        }}))
     B=&MakeVectorFromLinearColorRGB(Color=&RandomRangeLinearColor(Minimum=<rgba 0.7 0.7 0.7 1>, Maximum=<rgba 1 1 1 1>))
    )
    ScaleA=true
    ScaleRGB=true
   )
   /SetVariables(Particles.Sparks Color:LinearColor=&MultiplyLinearColor(
     Linear Color=$Particles.Initial.Sparks Color
     Scale Factor=&Vector4FromCurve(Vector4Curve=DataInterface() {propertyValues {AlphaCurve {keys [<at 0 1>,<at 0.3308 1>,<at 1 0>]}, BlueCurve {keys [<at 0 1>]}, GreenCurve {keys [<at 0 1>]}, RedCurve {keys [<at 0 1>]}}})
    )) {isSetParameters=true}
  ]
  rendererClasses [&NiagaraSpriteRendererProperties]
  renderers [
   {rendererClass=&NiagaraSpriteRendererProperties}
   {rendererClass=&NiagaraSpriteRendererProperties, rendererIndex=1}
  ]
 }
 Intro_FieryHeart {
  emitterSpawnScript [
   &InitializeAudioPlayer(Audio Player=DataInterface() {propertyValues {
      Attenuation=null
      Concurrency=null
      MaxPlaysPerTick=10
      SoundToPlay=/CosmeticCompanions/Assets/Audio_Shared/Interaction/Petting/SC_Mimosa_Interaction_React_Default
      bAllowLoopingOneShotSounds=false
      bLimitPlaysPerTick=true
      bOnlyActiveDuringGameplay=false
      bStopWhenComponentIsDestroyed=true
      ConfigurationUserParameter {parameter {name=None, typeDefHandle {registeredTypeIndex=1}, varData []}}
      ParameterNames []
     }})
   /Niagara/Modules/Audio/PlayAudio() {module=PlayAudio001}
  ]
  emitterUpdateScript [
   &EmitterState()
   &SpawnBurst_Instantaneous(Spawn Group=2, Use Spawn Probability=false, UseLoopCountLimit=false) {module=SpawnBurst_Instantaneous002}
   &SpawnBurst_Instantaneous(Spawn Count=3, Spawn Group=1, Use Spawn Probability=false, UseLoopCountLimit=false) {module=SpawnBurst_Instantaneous001}
   &SpawnBurst_Instantaneous(Use Spawn Probability=false, UseLoopCountLimit=false)
  ]
  particleSpawnScript [
   &InitializeParticle(
    Lifetime=&MergeHelper_FloatSwitch(
     0=$System.IntroLength
     1=&RandomRangeFloat(Minimum=0.25, Maximum=0.6)
     2=0.3
     3=?
     Index=&SpawnGroup()
     Num=::NewEnumerator1
    )
    Position:Vector3f=$System.BurstPosition
    Position Mode=::NewEnumerator1
    Sprite Rotation Angle=&MergeHelper_FloatSwitch(
     0=0
     1=&RandomRangeFloat(Minimum=0, Maximum=360)
     2=?
     3=?
     Index=&SpawnGroup()
     Num=::NewEnumerator0
    )
    Sprite Rotation Mode=::NewEnumerator0
    Sprite Size Mode=::NewEnumerator3
    Uniform Sprite Size=35
    UsePositionOffset=true
    Write Material Random=true
   )
   /SetVariables(Particles.VisibilityTag=&SpawnGroup()) {isSetParameters=true, module=SetVariables_#11}
   /SetVariables(
    Particles.Smoke Color:LinearColor=&RandomRangeLinearColor(Minimum=<rgba 0 0 0 1>, Maximum=<rgba 0.0156 0.0028 0 1>)
    Particles.Smoke Sprite Size:Vector2f=&Vector2DFromFloat(Value=&RandomRangeFloat(Minimum=30, Maximum=45))
   ) {isSetParameters=true}
   /SetVariables(Particles.Glow Color=<rgba 1 0 0.0099 0.25>, Particles.Glow Size=&Vector2DFromFloat(Value=35)) {isSetParameters=true}
  ]
  particleUpdateScript [
   &ParticleState()
   /Niagara/Modules/Update/SubUV/V2/SubUVAnimation(
    Direct Frame Index:NiagaraFloat=&Lerp_Float(
     A=0
     B=8
     Alpha=&FloatFromCurve(FloatCurve=DataInterface() {propertyValues {Curve {keys [
          {arriveTangent=2.3533, leaveTangentWeight=0.5751, time=0, value=0}
          {arriveTangent=1.0494, arriveTangentWeight=0.1033, leaveTangent=1.0494, time=1, value=1}
         ]}}})
    )
    "Sprite Renderer "=DataInterface() {propertyValues {SpriteRenderer=&self:Intro_FieryHeart_0.NiagaraSpriteRendererProperties_0}}
    SubUV Animation Mode:ENiagara_SubUVLookupModeV2=::NewEnumerator5
   )
   /SetVariables(
    Particles.Linear Color_Halo:LinearColor=&MakeLinearColorFromVectorAndFloat(
     "Vector (RGB)"=&MakeVectorFromLinearColorRGB(Color=<rgba 1 0.1765 0 1>)
     "Float (Alpha)"=&FloatFromCurve(
      FloatCurve=DataInterface() {propertyValues {Curve {keys [
          {arriveTangent=0.4476, interpMode=RCIM_Cubic, leaveTangent=0.4476, tangentMode=RCTM_User, time=0, value=0}
          {arriveTangent=0.3048, interpMode=RCIM_Cubic, leaveTangent=0.3048, time=0.6011, value=0.3}
          {arriveTangent=-0.7521, interpMode=RCIM_Cubic, leaveTangent=-0.7521, time=0.9841, value=0.3}
          <at 1 0 RCIM_Cubic>
         ]}}}
      Scale Curve=&FN_FrontEndFloatModifier(Gameplay=1, Front End=0.3)
     )
    )
    Particles.SpriteSize_Halo:Vector2f=&Multiply_Vector2D_ByFloat(
     Float=&FloatFromCurve(
      FloatCurve=DataInterface() {propertyValues {Curve {keys [<at 0 0>,<at 0.5 1>,<at 0.9853 1>,<at 1 0>]}}}
      Scale Curve=&FN_FrontEndFloatModifier(Gameplay=28.5083, Front End=28)
     )
     Vector2D=<1 1>
    )
   ) {isSetParameters=true, module=SetVariables_#12}
   /SetVariables(
    Particles.LinearColor_Electric:LinearColor=&Multiply_LinearColorByFloat(
     Linear Color=&ColorFromCurve(ColorCurve=DataInterface() {propertyValues {
        AlphaCurve {keys [<at 0.9116 1 RCIM_Cubic>]}
        BlueCurve {keys [<at 0.0586 1 RCIM_Constant>,<at 0.1939 1 RCIM_Constant>,<at 0.3325 1 RCIM_Constant>,<at 0.493 1 RCIM_Constant>]}
        GreenCurve {keys [<at 0.0586 1 RCIM_Constant>,<at 0.1939 1 RCIM_Constant>,<at 0.3325 1 RCIM_Constant>,<at 0.493 1 RCIM_Constant>]}
        RedCurve {keys [<at 0.0586 1 RCIM_Constant>,<at 0.1939 1 RCIM_Constant>,<at 0.3325 1 RCIM_Constant>,<at 0.493 1 RCIM_Constant>]}
       }})
     Float=1
    )
    Particles.SpriteSize_Electric:Vector2f=&Vector2DFromFloat(Value=&FloatFromCurve(
      FloatCurve=DataInterface() {propertyValues {Curve {keys [
          {arriveTangent=21.5218, interpMode=RCIM_Cubic, leaveTangent=21.5218, tangentMode=RCTM_User, time=0, value=0.2}
          {arriveTangent=4.6934, interpMode=RCIM_Cubic, leaveTangent=4.6934, tangentMode=RCTM_User, time=0.055, value=1}
          {
           arriveTangent=-0.05
           arriveTangentWeight=0.0271
           interpMode=RCIM_Cubic
           leaveTangent=-0.05
           leaveTangentWeight=0.2796
           tangentMode=RCTM_User
           time=0.3114
           value=0.999
          }
          {arriveTangent=-0.05, arriveTangentWeight=0.2667, interpMode=RCIM_Cubic, leaveTangent=-0.05, tangentMode=RCTM_User, time=1, value=1}
         ]}}}
      Scale Curve=27
     ))
   ) {isSetParameters=true, module=SetVariables_#13}
   /SetVariables(Particles.Color_Shatter:LinearColor=&ColorFromCurve(
     ColorCurve=DataInterface() {propertyValues {
       AlphaCurve {keys [<at 0.3053 1>,<at 0.5076 0 RCIM_Cubic>]}
       BlueCurve {keys [<at 0.5115 0 RCIM_Cubic>]}
       GreenCurve {keys [<at 0.5115 0 RCIM_Cubic>]}
       RedCurve {keys [<at 0.5115 0 RCIM_Cubic>]}
      }}
     CurveIndex=&Multiply_Float(A=$Particles.NormalizedAge, B=1)
    )) {isSetParameters=true, module=SetVariables_#14}
   /SetVariables(Particles.Wiggle:NiagaraFloat=&Waveform(
     Global Amplitude Scale=&MergeHelper_FloatSwitch(0=0.6, 1=0, 2=?, 3=?, Index=&SpawnGroup(), Num=::NewEnumerator0)
     "Waveform [1]"=::NewEnumerator6
     "[1] Frequency"=&FloatFromCurve(FloatCurve=DataInterface() {propertyValues {Curve {keys [
          {interpMode=RCIM_Cubic, tangentMode=RCTM_User, time=0, value=5}
          {arriveTangent=5.0571, interpMode=RCIM_Cubic, leaveTangent=5.0571, tangentMode=RCTM_User, time=0.7927, value=6.8835}
          {arriveTangent=-2.4916, interpMode=RCIM_Cubic, leaveTangent=-2.4916, tangentMode=RCTM_User, time=0.8947, value=7}
          {arriveTangent=30.9125, interpMode=RCIM_Cubic, leaveTangent=30.9125, tangentMode=RCTM_User, time=0.972, value=9}
         ]}}})
     "[1] Amplitude Min/Max"=&Multiply_Vector2D_ByFloat(
      Float=&FloatFromCurve(
       FloatCurve=DataInterface() {propertyValues {Curve {defaultValue=1, keys [{leaveTangentWeight=0.3515, time=0, value=1}, {arriveTangent=3.3152, leaveTangent=3.3152, time=0.9761, value=2}]}}}
       Scale Curve=0.75
      )
      Vector2D=<-1 1>
     )
    )) {isSetParameters=true}
   /SetVariables(Particles.Position=&AddVectorToPosition(
     Position=$Particles.Position
     Vector=&MakeVector(
      X=&Lerp_Float(
       A=0
       B=$Particles.Wiggle
       Alpha=&FloatFromCurve(FloatCurve=DataInterface() {propertyValues {Curve {defaultValue=1, keys [{tangentMode=RCTM_User, time=0.4, value=0}, {tangentMode=RCTM_User, time=0.52, value=1}]}}})
      )
      Y=0
      Z=0
     )
    )) {isSetParameters=true}
   /SetVariables(
    Particles.Smoke Color:LinearColor=&MultiplyLinearColor(
     Linear Color=$Particles.Initial.Smoke Color
     Scale Factor=&Vector4FromCurve(Vector4Curve=DataInterface() {propertyValues {
        AlphaCurve {keys [<at 0 0>,<at 0.0796 1>,<at 1 0>]}
        BlueCurve {keys [<at 0 2.5>,<at 1 1>]}
        GreenCurve {keys [<at 0 2.5>,<at 1 1>]}
        RedCurve {keys [<at 0 2.5>,<at 1 1>]}
       }})
    )
    Particles.Smoke Sprite Size:Vector2f=&Multiply_Vector2D(
     A=$Particles.Initial.Smoke Sprite Size
     B=&Vector2DFromFloat(Value=&FloatFromCurve(FloatCurve=DataInterface() {propertyValues {Curve {keys [<at 0 0.8>,<at 1 1.5>]}}}, Scale Curve=1.156))
    )
   ) {isSetParameters=true}
   /SetVariables(
    Particles.Glow Color=&MultiplyLinearColor(
     Linear Color=$Particles.Initial.Glow Color
     Scale Factor=&Vector4FromCurve(Vector4Curve=DataInterface() {propertyValues {
        AlphaCurve {keys [<at 0 1>,<at 0.3667 1>,<at 1 0>]}
        BlueCurve {keys [<at 0 1>,<at 1 1>]}
        GreenCurve {keys [<at 0 1>,<at 1 1>]}
        RedCurve {keys [<at 0 1>,<at 1 1>]}
       }})
    )
    Particles.Glow Size=&Multiply_Vector2D(
     A=$Particles.Initial.Glow Size
     B=&Vector2DFromFloat(Value=&FloatFromCurve(FloatCurve=DataInterface() {propertyValues {Curve {keys [<at 0 0>,<at 0.5 1>]}}}, Scale Curve=1.3))
    )
   ) {isSetParameters=true}
   &DynamicMaterialParameters(Write Parameter Index 0=::NewEnumerator0)
  ]
  rendererClasses [&NiagaraSpriteRendererProperties]
  renderers [
   {rendererClass=&NiagaraSpriteRendererProperties}
   {rendererClass=&NiagaraSpriteRendererProperties, rendererIndex=1}
   {rendererClass=&NiagaraSpriteRendererProperties, rendererIndex=2}
   {rendererClass=&NiagaraSpriteRendererProperties, rendererIndex=3}
  ]
 }
 SmokeLWE {rendererClasses [], renderers []}
 SmokeLWE_Burst {rendererClasses [], renderers []}
}
summary {
 systemName=NS_ReactFX_RareKrow
 emitters [
  {emitterName=Intro_FieryHeart, rendererClasses [&NiagaraSpriteRendererProperties]}
  {emitterName=Burst_FieryHeart, rendererClasses [&NiagaraSpriteRendererProperties]}
  {emitterName=SmokeLWE, rendererClasses []}
  {emitterName=SmokeLWE_Burst, rendererClasses []}
  {emitterName=Flame_Sprites, rendererClasses [&NiagaraSpriteRendererProperties]}
  {emitterName=Burst, rendererClasses [&NiagaraSpriteRendererProperties, &NiagaraMeshRendererProperties]}
 ]
 userVariables [
  User.bIsFrontEnd {type=NiagaraBool}
  User.bIsFrontEndPreview {type=NiagaraBool}
 ]
}
