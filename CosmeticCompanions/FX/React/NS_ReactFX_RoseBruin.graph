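# Niagara reaction effect for the RoseBruin cosmetic companion: a heart swells, shakes and pops into pink hearts and sparkles.
# NewIntro plays the Mimosa react sound and animates the heart through System.IntroLength, when NewPop adds the burst flash, flare and glow.
# HeartsSparkles bursts three spawn groups on a fixed delay, separated by VisibilityTag and drifting under curl noise, gravity and attraction.
# Shared system variables set position and size, and time-of-day colors adjust brightness.
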NiagaraSystem /CosmeticCompanions/FX/React/NS_ReactFX_RoseBruin
dependencies {
 usedDataInterfaces [&NiagaraDataInterfaceCurve, &NiagaraDataInterfaceColorCurve, &NiagaraDataInterfaceAudioPlayer]
 usedDynamicInputs [/Niagara/DynamicInputs/TypeConversions/MakeCustomPositionFromBool, &FN_FrontEndVector3Modifier]
 usedModules [
  &self:SystemSpawnScript.SystemScriptSource.SystemScriptGraph.NiagaraNodeAssignment_0.SetVariables_#1
  &self:SystemSpawnScript.SystemScriptSource.SystemScriptGraph.NiagaraNodeAssignment_1.SetVariables_#2
  &self:SystemSpawnScript.SystemScriptSource.SystemScriptGraph.NiagaraNodeAssignment_3.SetVariables_#3
  &self:SystemSpawnScript.SystemScriptSource.SystemScriptGraph.NiagaraNodeAssignment_4.SetVariables_#4
  &self:SystemSpawnScript.SystemScriptSource.SystemScriptGraph.NiagaraNodeAssignment_5.SetVariables_#5
  &self:SystemSpawnScript.SystemScriptSource.SystemScriptGraph.NiagaraNodeAssignment_2.SetVariables_#6
  &SystemState
 ]
}
emitters {
 HeartsSparkles {
  emitterUpdateScript [
   &EmitterState()
   &SpawnBurst_Instantaneous(Spawn Count=5, Spawn Time=0.825, Use Spawn Probability=false, UseLoopCountLimit=false)
   &SpawnBurst_Instantaneous(
    Spawn Count=&RandomRangeInt(Minimum=11, Maximum=13)
    Spawn Group=1
    Spawn Time=0.825
    Use Spawn Probability=false
    UseLoopCountLimit=false
   ) {module=SpawnBurst_Instantaneous001}
   &SpawnBurst_Instantaneous(
    Spawn Count=&RandomRangeInt(Minimum=5, Maximum=7)
    Spawn Group=2
    Spawn Time=0.825
    Use Spawn Probability=false
    UseLoopCountLimit=false
   ) {module=SpawnBurst_Instantaneous002}
   /SetVariables(Emitter.Heart Color:LinearColor=<rgba 2 0.3055 0.6324 1>) {isSetParameters=true}
  ]
  particleSpawnScript [
   /SetVariables(Particles.VisibilityTag=&SpawnGroup()) {isSetParameters=true}
   &InitializeParticle(
    Color=&MergeHelper_ColorSwitch(
     0=&RandomRangeLinearColor(Minimum=<rgba 2 0.6743 0.9242 1>, Maximum=<rgba 2 0.2605 0.5097 1>)
     1=&RandomRangeLinearColor(Minimum=<rgba 3 1.681 2.0591 1>, Maximum=<rgba 5 0.5489 1.6596 1>)
     2=?
     3=?
     Index=&SpawnGroup()
     Num=::NewEnumerator0
    )
    Lifetime=&RandomRangeFloat(Minimum=0.6, Maximum=0.9)
    Position Offset=$System.BurstPosition
    Sprite Rotation Mode=::NewEnumerator2
    Sprite Size Mode=::NewEnumerator3
    Uniform Sprite Size=&MergeHelper_FloatSwitch(
     0=&Multiply_Float(A=$System.HeartSize, B=&RandomRangeFloat(Minimum=4, Maximum=16))
     1=&RandomRangeFloat(Minimum=8, Maximum=22)
     2=?
     3=?
     Index=&SpawnGroup()
     Num=::NewEnumerator0
    )
    UsePositionOffset=true
    Write Material Random=true
   )
   &ShapeLocation(
    MaskSpawnGroup=false
    Non Uniform Scale=<1.5 1 1>
    Offset=<0 0 -10>
    Offset Mode=::NewEnumerator0
    Override Seed=false
    Sphere Radius=13
    Sphere Surface Distribution=1
    UseOwnerScale=true
   )
   &AddVelocity(Velocity=&RandomRangeVector(Minimum=<-30 -30 100>, Maximum=<30 30 300>)) {module=AddVelocity001}
  ]
  particleUpdateScript [
   &ParticleState()
   &ScaleSpriteSize(Uniform Curve Sprite Scale=DataInterface() {propertyValues {Curve {
       defaultValue=0
       keys [
        {interpMode=RCIM_Cubic, tangentMode=RCTM_User, time=0.7839, value=1}
        {arriveTangent=-9.8207, interpMode=RCIM_Cubic, leaveTangent=-9.8207, tangentMode=RCTM_User, time=1, value=0}
       ]
      }}})
   /SetVariables(
    Particles.Bokeh Flicker:NiagaraFloat=&Waveform(
     "Waveform [1]"=::NewEnumerator0
     "[1] Frequency"=&RandomRangeFloat(Minimum=1, Maximum=5)
     "[1] Amplitude Min/Max"=<0.1 1>
    )
    Particles.Flicker:NiagaraFloat=&Waveform(
     "Waveform [1]"=::NewEnumerator0
     "[1] Frequency"=&RandomRangeFloat(Minimum=3, Maximum=5)
     "[1] Amplitude Min/Max"=<0 1>
    )
   ) {isSetParameters=true}
   &GravityForce(Gravity=&RandomRangeVector(Minimum=<0 0 -300>, Maximum=<0 0 -500>))
   &Drag(
    Drag=&Multiply_Float(
     A=&RandomRangeFloat(Minimum=1, Maximum=5)
     B=&FloatFromCurve(FloatCurve=DataInterface() {propertyValues {Curve {defaultValue=1, keys [{tangentMode=RCTM_User, time=0.2, value=2}, {tangentMode=RCTM_User, time=0.65, value=0.5}]}}})
    )
    Use Linear Drag=true
    Use Rotational Drag=true
   )
   &CurlNoiseForce(
    Bias Noise=false
    Noise Frequency=200
    Noise Strength=&RandomRangeFloat(Minimum=200, Maximum=400)
    Pan Noise=false
   )
   &PointAttractionForce(
    Attraction Strength=&FloatFromCurve(
     FloatCurve=DataInterface() {propertyValues {Curve {
        defaultValue=0
        keys [{interpMode=RCIM_Cubic, tangentMode=RCTM_User, time=0, value=0}, {interpMode=RCIM_Cubic, tangentMode=RCTM_User, time=0.4, value=1}]
       }}}
     Scale Curve=500
    )
    Attractor Position Offset:Vector3f=<0 0 -300>
    Kill Within Radius=false
    Module.Use Falloff=true
    UseAttractorOffset=true
   )
   &SolveForcesAndVelocity(Clamp Velocity=false, Limit Acceleration=false)
   &ScaleColor(
    Scale Alpha=&MergeHelper_FloatSwitch(
     0=&FloatFromCurve(FloatCurve=DataInterface() {propertyValues {Curve {
         defaultValue=1
         keys [
          {arriveTangent=24.9349, interpMode=RCIM_Cubic, leaveTangent=24.9349, tangentMode=RCTM_User, time=0, value=0}
          {arriveTangent=0.7276, interpMode=RCIM_Cubic, leaveTangent=0.7276, tangentMode=RCTM_User, time=0.1008, value=1.0033}
          {arriveTangent=-1.8102, interpMode=RCIM_Cubic, leaveTangent=-1.8102, tangentMode=RCTM_User, time=0.5915, value=0.8526}
          {arriveTangent=-0.0405, interpMode=RCIM_Cubic, leaveTangent=-0.0405, tangentMode=RCTM_User, time=1, value=0}
         ]
        }}})
     1=&Multiply_Float(
      A=&FloatFromCurve(FloatCurve=DataInterface() {propertyValues {Curve {
          defaultValue=1
          keys [
           {arriveTangent=27.2232, interpMode=RCIM_Cubic, leaveTangent=27.2232, tangentMode=RCTM_User, time=0, value=0}
           {arriveTangent=0.4162, interpMode=RCIM_Cubic, leaveTangent=0.4162, tangentMode=RCTM_User, time=0.0689, value=1.2342}
           {arriveTangent=-0.0405, interpMode=RCIM_Cubic, leaveTangent=-0.0405, tangentMode=RCTM_User, time=1, value=0}
          ]
         }}})
      B=$Particles.Flicker
     )
     2=?
     3=?
     Index=&SpawnGroup()
     Num=::NewEnumerator0
    )
    Scale RGB=&MergeHelper_VectorSwitch(
     0=<1 1 1>
     1=&MakeVectorFromLinearColorRGB(Color=&ColorFromCurve(ColorCurve=DataInterface() {propertyValues {
         AlphaCurve {keys []}
         BlueCurve {keys [<at 0.0637 3>,<at 0.1009 1>,<at 0.3329 1>,<at 0.9944 0.1323>]}
         GreenCurve {keys [<at 0.0637 3>,<at 0.1009 1>,<at 0.3329 1>,<at 0.9944 0>]}
         RedCurve {keys [<at 0.0637 3>,<at 0.1009 1>,<at 0.3329 1>,<at 0.9944 0.3646>]}
        }}))
     2=?
     3=?
     Index=&SpawnGroup()
     Num=::NewEnumerator0
    )
    ScaleA=true
    ScaleRGB=true
   )
   /SetVariables(
    Particles.Bokeh Color:LinearColor=&MakeLinearColorFromVectorAndFloat(
     "Vector (RGB)"=&MakeVectorFromLinearColorRGB(Color=&LerpLinearColors(
       StartColor=&RandomRangeLinearColor(Minimum=<rgba 1 0.4275 0.6125 1>, Maximum=<rgba 1 0.0585 0.3572 1>)
       EndColor=&RandomRangeLinearColor(Minimum=<rgba 0.6213 0.1279 1 1>, Maximum=<rgba 0.2135 0.0125 0.0763 1>)
       LerpFactor=&FloatFromCurve(FloatCurve=DataInterface() {propertyValues {Curve {
           defaultValue=1
           keys [
            {interpMode=RCIM_Cubic, tangentMode=RCTM_User, time=0, value=0}
            {arriveTangent=1.2118, interpMode=RCIM_Cubic, leaveTangent=1.2118, tangentMode=RCTM_User, time=1, value=1}
           ]
          }}})
      ))
     "Float (Alpha)"=&Multiply_Float(
      A=&FloatFromCurve(
       FloatCurve=DataInterface() {propertyValues {Curve {
          defaultValue=1
          keys [{interpMode=RCIM_Cubic, tangentMode=RCTM_User, time=0, value=1}, {interpMode=RCIM_Cubic, tangentMode=RCTM_User, time=1, value=0}]
         }}}
       Scale Curve=0.3
      )
      B=$Particles.Bokeh Flicker
     )
    )
    Particles.Bokeh Size:Vector2f=&Multiply_Vector2D_ByFloat(
     Float=&FloatFromCurve(FloatCurve=DataInterface() {propertyValues {Curve {
         defaultValue=1
         keys [
          {arriveTangent=40.8712, interpMode=RCIM_Cubic, leaveTangent=40.8712, tangentMode=RCTM_User, time=0, value=0}
          {arriveTangent=0.7672, interpMode=RCIM_Cubic, leaveTangent=0.7672, tangentMode=RCTM_User, time=0.0663, value=1.0039}
          {arriveTangent=-1.712, interpMode=RCIM_Cubic, leaveTangent=-1.712, tangentMode=RCTM_User, time=0.3106, value=0.656}
          {arriveTangent=-0.0405, interpMode=RCIM_Cubic, leaveTangent=-0.0405, tangentMode=RCTM_User, time=1, value=0.25}
         ]
        }}})
     Vector2D=&Vector2DFromFloat(Value=&RandomRangeFloat(Minimum=5, Maximum=25))
    )
   ) {isSetParameters=true}
   /SetVariables(Particles.Vector 4:Vector4f=&MakeVector4(
     X=0
     Y=0
     Z=&Multiply_Float(
      A=1
      B=&FloatFromCurve(FloatCurve=DataInterface() {propertyValues {Curve {
          defaultValue=1
          keys [
           {arriveTangent=24.9349, interpMode=RCIM_Cubic, leaveTangent=24.9349, tangentMode=RCTM_User, time=0, value=0}
           {arriveTangent=0.7276, interpMode=RCIM_Cubic, leaveTangent=0.7276, tangentMode=RCTM_User, time=0.0783, value=0.9655}
           {arriveTangent=-1.8102, interpMode=RCIM_Cubic, leaveTangent=-1.8102, tangentMode=RCTM_User, time=0.6189, value=0.8021}
           {arriveTangent=-2.1697, interpMode=RCIM_Cubic, leaveTangent=-2.1697, tangentMode=RCTM_User, time=0.8033, value=0.3088}
           {arriveTangent=-0.0405, interpMode=RCIM_Cubic, leaveTangent=-0.0405, tangentMode=RCTM_User, time=1, value=0}
          ]
         }}})
     )
     W=0
    )) {isSetParameters=true}
  ]
  rendererClasses [&NiagaraSpriteRendererProperties]
  renderers [
   {rendererClass=&NiagaraSpriteRendererProperties}
   {rendererClass=&NiagaraSpriteRendererProperties, rendererIndex=1}
   {rendererClass=&NiagaraSpriteRendererProperties, rendererIndex=2}
  ]
 }
 NewIntro {
  emitterSpawnScript [
   &InitializeAudioPlayer(Audio Player=DataInterface() {propertyValues {
      Attenuation=null
      Concurrency=null
      MaxPlaysPerTick=10
      SoundToPlay=/CosmeticCompanions/Assets/Audio_Shared/Interaction/Petting/SC_Mimosa_Interaction_React_Default
      bAllowLoopingOneShotSounds=false
      bLimitPlaysPerTick=true
      bOnlyActiveDuringGameplay=false
      bStopWhenComponentIsDestroyed=true
      ConfigurationUserParameter {parameter {name=None, typeDefHandle {registeredTypeIndex=1}, varData []}}
      ParameterNames []
     }})
   /Niagara/Modules/Audio/PlayAudio() {module=PlayAudio001}
  ]
  emitterUpdateScript [
   &EmitterState()
   &SpawnBurst_Instantaneous(Use Spawn Probability=false, UseLoopCountLimit=false)
   /SetVariables(
    Emitter.Cracked Heart Color:LinearColor=&FNTimeOfDayColor(Night Color=<rgba 5 0 0.4264 2>, Morning Color=<rgba 5 0 0.4264 2>, Day Color=<rgba 5 0.0781 0.4979 2>, Evening Color=<rgba 5 0 0.4264 2>)
    Emitter.Heart Color:LinearColor=<rgba 1 1 1 2>
    Emitter.Lightning Color:LinearColor=&ColorFromCurve(
     ColorCurve=DataInterface() {propertyValues {
       AlphaCurve {keys [<at 0 1 RCIM_Cubic>]}
       BlueCurve {keys [<at 0 1.0736>, {arriveTangent=3.4057, interpMode=RCIM_Cubic, leaveTangent=3.4057, time=0.0976, value=1.3651}, <at 0.4188 2.5 RCIM_Cubic>]}
       GreenCurve {keys [<at 0 0.25>, {arriveTangent=5.3721, interpMode=RCIM_Cubic, leaveTangent=5.3721, time=0.0976, value=0.6662}, <at 0.4188 2.5 RCIM_Cubic>]}
       RedCurve {keys [<at 0 3>, {arriveTangent=-1.1938, interpMode=RCIM_Cubic, leaveTangent=-1.1938, time=0.0976, value=3}, <at 0.4188 2.5 RCIM_Cubic>]}
      }}
     CurveIndex=$Emitter.NormalizedLoopAge
    )
    Emitter.Outline Color:LinearColor=&FNTimeOfDayColor(Night Color=<rgba 1 1 1 1>, Morning Color=<rgba 1.4 1.4 1.4 1>, Day Color=<rgba 3 3 3 1>, Evening Color=<rgba 1.4 1.4 1.4 1>)
   ) {isSetParameters=true, module=SetVariables_#7}
  ]
  particleSpawnScript [&InitializeParticle(
    Color Mode=::NewEnumerator0
    Lifetime=$System.IntroLength
    Position Offset=$System.BurstPosition
    Sprite Rotation Mode=::NewEnumerator0
    Sprite Size Mode=::NewEnumerator3
    Uniform Sprite Size=&Multiply_Float(A=15, B=$System.HeartSize)
    UsePositionOffset=true
    Write Material Random=true
   )]
  particleUpdateScript [
   &ParticleState()
   /SetVariables(Particles.Wiggle:NiagaraFloat=&Waveform(
     "Waveform [1]"=::NewEnumerator0
     "[1] Frequency"=&FloatFromCurve(FloatCurve=DataInterface() {propertyValues {Curve {keys [
          {interpMode=RCIM_Cubic, tangentMode=RCTM_User, time=0, value=5}
          {arriveTangent=5.0571, interpMode=RCIM_Cubic, leaveTangent=5.0571, tangentMode=RCTM_User, time=0.7927, value=6.8835}
          {arriveTangent=-2.4916, interpMode=RCIM_Cubic, leaveTangent=-2.4916, tangentMode=RCTM_User, time=0.8947, value=7}
          {arriveTangent=30.9125, interpMode=RCIM_Cubic, leaveTangent=30.9125, tangentMode=RCTM_User, time=0.972, value=9}
         ]}}})
     "[1] Amplitude Min/Max"=&Multiply_Vector2D_ByFloat(
      Float=&FloatFromCurve(
       FloatCurve=DataInterface() {propertyValues {Curve {
          defaultValue=1
          keys [
           {interpMode=RCIM_Cubic, leaveTangentWeight=0.3515, tangentMode=RCTM_User, time=0, value=1}
           {arriveTangent=3.3152, interpMode=RCIM_Cubic, leaveTangent=3.3152, tangentMode=RCTM_User, time=0.9761, value=2}
          ]
         }}}
       Scale Curve=0.75
      )
      Vector2D=<-1 1>
     )
    )) {isSetParameters=true}
   &ScaleSpriteSize(
    Scale Sprite Size Mode=::NewEnumerator0
    Uniform Scale Factor=&Add_Float(
     A=0.5
     B=&FloatFromCurve(FloatCurve=DataInterface() {propertyValues {Curve {
         defaultValue=0
         keys [
          {arriveTangent=0.8193, interpMode=RCIM_Cubic, leaveTangent=0.8193, tangentMode=RCTM_User, time=0, value=0}
          {arriveTangent=2.055, interpMode=RCIM_Cubic, leaveTangent=2.055, tangentMode=RCTM_User, time=0.7247, value=0.9545}
          {arriveTangent=-0.9053, interpMode=RCIM_Cubic, leaveTangent=-0.9053, tangentMode=RCTM_User, time=0.8783, value=2}
          {arriveTangent=-0.1202, interpMode=RCIM_Cubic, leaveTangent=-0.1202, tangentMode=RCTM_User, time=1, value=1}
         ]
        }}})
    )
   ) {module=ScaleSpriteSize002}
   &SolveForcesAndVelocity(Clamp Velocity=false, Limit Acceleration=false)
   &DynamicMaterialParameters(
    Index 0 Param 1=&FloatFromCurve(FloatCurve=DataInterface() {propertyValues {Curve {
        defaultValue=0
        keys [
         {arriveTangent=9.0725, interpMode=RCIM_Cubic, leaveTangent=9.0725, tangentMode=RCTM_User, time=0, value=0}
         {arriveTangent=-0.2441, leaveTangent=-0.2441, time=0.1408, value=0.9847}
         {arriveTangent=-1.146, leaveTangent=-1.146, time=0.95, value=1}
         {arriveTangent=-0.0405, interpMode=RCIM_Cubic, leaveTangent=-0.0405, tangentMode=RCTM_User, time=1, value=0}
        ]
       }}})
    Index 0 Param 2=&FloatFromCurve(FloatCurve=DataInterface() {propertyValues {Curve {
        defaultValue=0
        keys [
         {arriveTangent=9.0725, interpMode=RCIM_Cubic, leaveTangent=9.0725, tangentMode=RCTM_User, time=0, value=0}
         {arriveTangent=-0.2441, leaveTangent=-0.2441, time=0.1408, value=0.9847}
         {arriveTangent=-1.146, leaveTangent=-1.146, time=0.95, value=1}
         {arriveTangent=-0.0405, interpMode=RCIM_Cubic, leaveTangent=-0.0405, tangentMode=RCTM_User, time=1, value=0}
        ]
       }}})
    Index 0 Param 3=&FloatFromCurve(FloatCurve=DataInterface() {propertyValues {Curve {
        defaultValue=0
        keys [
         {arriveTangent=9.0725, interpMode=RCIM_Cubic, leaveTangent=9.0725, tangentMode=RCTM_User, time=0, value=0}
         {arriveTangent=-0.2441, leaveTangent=-0.2441, time=0.1408, value=0.9847}
         {arriveTangent=-1.146, leaveTangent=-1.146, time=0.95, value=1}
         {arriveTangent=-0.0405, interpMode=RCIM_Cubic, leaveTangent=-0.0405, tangentMode=RCTM_User, time=1, value=0}
        ]
       }}})
    Index 0 Param 4=&FloatFromCurve(FloatCurve=DataInterface() {propertyValues {Curve {
        defaultValue=0
        keys [
         {arriveTangent=9.0725, interpMode=RCIM_Cubic, leaveTangent=9.0725, tangentMode=RCTM_User, time=0, value=0}
         {arriveTangent=-0.2441, leaveTangent=-0.2441, time=0.1408, value=0.9847}
         {arriveTangent=-1.146, leaveTangent=-1.146, time=0.95, value=1}
         {arriveTangent=-0.0405, interpMode=RCIM_Cubic, leaveTangent=-0.0405, tangentMode=RCTM_User, time=1, value=0}
        ]
       }}})
    Module.Index0Float0Write=true
    Module.Index0Float1Write=true
    Module.Index0Float2Write=true
    Module.Index0Float3Write=true
   )
   /SetVariables(Particles.Position=&AddVectorToPosition(
     Position=$Particles.Position
     Vector=&MakeVector(
      X=&Lerp_Float(
       A=0
       B=$Particles.Wiggle
       Alpha=&FloatFromCurve(FloatCurve=DataInterface() {propertyValues {Curve {defaultValue=1, keys [{tangentMode=RCTM_User, time=0.4, value=0}, {tangentMode=RCTM_User, time=0.52, value=1}]}}})
      )
      Y=0
      Z=0
     )
    )) {isSetParameters=true}
   &ScaleColor(
    Scale Alpha=&FloatFromCurve(FloatCurve=DataInterface() {propertyValues {Curve {
        defaultValue=0
        keys [
         {arriveTangent=9.0725, interpMode=RCIM_Cubic, leaveTangent=9.0725, tangentMode=RCTM_User, time=0, value=0}
         {arriveTangent=-0.2441, leaveTangent=-0.2441, time=0.1408, value=0.9847}
         {arriveTangent=-1.146, leaveTangent=-1.146, time=0.95, value=1}
         {arriveTangent=-0.0405, interpMode=RCIM_Cubic, leaveTangent=-0.0405, tangentMode=RCTM_User, time=1, value=0}
        ]
       }}})
    ScaleA=true
    ScaleRGB=true
   )
  ]
  rendererClasses [&NiagaraSpriteRendererProperties]
  renderers [{rendererClass=&NiagaraSpriteRendererProperties}]
 }
 NewPop {
  emitterUpdateScript [
   &EmitterState()
   &SpawnBurst_Instantaneous(Spawn Time=&Subtract_Float(A=$System.IntroLength, B=0), Use Spawn Probability=false, UseLoopCountLimit=false)
   /SetVariables(
    Emitter.Cracked Heart Color:LinearColor=&ColorFromCurve(
     ColorCurve=DataInterface() {propertyValues {
       AlphaCurve {keys [<at 0 1 RCIM_Cubic>]}
       BlueCurve {keys [<at 0.0057 1>,<at 0.0511 0.3784 RCIM_Cubic>,<at 0.0953 0.2232 RCIM_Cubic>]}
       GreenCurve {keys [<at 0.0057 1>,<at 0.0511 0.2656 RCIM_Cubic>,<at 0.0953 0.0823 RCIM_Cubic>]}
       RedCurve {keys [<at 0.0057 1>,<at 0.0511 1 RCIM_Cubic>,<at 0.0953 1 RCIM_Cubic>]}
      }}
     CurveIndex=$Emitter.NormalizedLoopAge
    )
    Emitter.Heart Color:LinearColor=&ColorFromCurve(
     ColorCurve=DataInterface() {propertyValues {
       AlphaCurve {keys [<at 0 1 RCIM_Cubic>]}
       BlueCurve {keys [<at 0.0057 1>,<at 0.0511 0.3784 RCIM_Cubic>,<at 0.0953 0.2232 RCIM_Cubic>]}
       GreenCurve {keys [<at 0.0057 1>,<at 0.0511 0.2656 RCIM_Cubic>,<at 0.0953 0.0823 RCIM_Cubic>]}
       RedCurve {keys [<at 0.0057 1>,<at 0.0511 1 RCIM_Cubic>,<at 0.0953 1 RCIM_Cubic>]}
      }}
     CurveIndex=$Emitter.NormalizedLoopAge
    )
    Emitter.Lightning Color:LinearColor=&ColorFromCurve(
     ColorCurve=DataInterface() {propertyValues {
       AlphaCurve {keys [<at 0 1 RCIM_Cubic>]}
       BlueCurve {keys [<at 0.0233 15>,<at 0.0549 10 RCIM_Cubic>,<at 0.0763 0.4798>]}
       GreenCurve {keys [<at 0.0233 15>,<at 0.0549 10 RCIM_Cubic>,<at 0.0763 0.3854>]}
       RedCurve {keys [<at 0.0233 15>,<at 0.0549 10 RCIM_Cubic>,<at 0.0763 1>]}
      }}
     CurveIndex=$Emitter.NormalizedLoopAge
    )
    Emitter.Outline Color:LinearColor=&ColorFromCurve(
     ColorCurve=DataInterface() {propertyValues {
       AlphaCurve {keys [<at 0 1 RCIM_Cubic>]}
       BlueCurve {keys [<at 0.0183 15>,<at 0.0473 10 RCIM_Cubic>,<at 0.1798 0.4798>]}
       GreenCurve {keys [<at 0.0183 15>,<at 0.0473 10 RCIM_Cubic>,<at 0.1798 0.3854>]}
       RedCurve {keys [<at 0.0183 15>,<at 0.0473 10 RCIM_Cubic>,<at 0.1798 1>]}
      }}
     CurveIndex=$Emitter.NormalizedLoopAge
    )
   ) {isSetParameters=true, module=SetVariables_#7}
  ]
  particleSpawnScript [&InitializeParticle(
    Color Mode=::NewEnumerator0
    Lifetime=0.2
    Position Offset=$System.BurstPosition
    Sprite Rotation Mode=::NewEnumerator0
    Sprite Size Mode=::NewEnumerator3
    Uniform Sprite Size=&Multiply_Float(A=15, B=$System.HeartSize)
    UsePositionOffset=true
    Write Material Random=true
   )]
  particleUpdateScript [
   &ParticleState()
   &ScaleSpriteSize(
    Scale Sprite Size Mode=::NewEnumerator0
    Uniform Scale Factor=&FloatFromCurve(FloatCurve=DataInterface() {propertyValues {Curve {keys [
         {arriveTangent=3.3458, interpMode=RCIM_Cubic, leaveTangent=3.3458, tangentMode=RCTM_User, time=0, value=1}
         {arriveTangent=4.5028, interpMode=RCIM_Cubic, leaveTangent=12.9417, tangentMode=RCTM_Break, time=0.15, value=1.471}
         {arriveTangent=1.8787, interpMode=RCIM_Cubic, leaveTangent=1.8787, time=0.175, value=3.9696}
         {arriveTangent=-0.2182, interpMode=RCIM_Cubic, leaveTangent=3.345, tangentMode=RCTM_Break, time=0.2, value=1.5649}
         {arriveTangent=1.094, interpMode=RCIM_Cubic, leaveTangent=1.6744, tangentMode=RCTM_Break, time=0.9989, value=3.0533}
        ]}}})
   ) {module=ScaleSpriteSize002}
   &SolveForcesAndVelocity(Clamp Velocity=false, Limit Acceleration=false)
   &DynamicMaterialParameters(
    Index 0 Param 1=&FloatFromCurve(FloatCurve=DataInterface() {propertyValues {Curve {
        defaultValue=0
        keys [
         {arriveTangent=-4.0077, interpMode=RCIM_Cubic, leaveTangent=-4.0077, tangentMode=RCTM_User, time=0.0275, value=1}
         {arriveTangent=-0.0705, interpMode=RCIM_Cubic, leaveTangent=-0.0705, tangentMode=RCTM_User, time=0.6, value=0}
        ]
       }}})
    Index 0 Param 2=&FloatFromCurve(FloatCurve=DataInterface() {propertyValues {Curve {
        defaultValue=0
        keys [
         {arriveTangent=-4.0077, interpMode=RCIM_Cubic, leaveTangent=-4.0077, tangentMode=RCTM_User, time=0.0275, value=1}
         {arriveTangent=-0.0705, interpMode=RCIM_Cubic, leaveTangent=-0.0705, tangentMode=RCTM_User, time=0.6, value=0}
        ]
       }}})
    Index 0 Param 3=&FloatFromCurve(FloatCurve=DataInterface() {propertyValues {Curve {
        defaultValue=0
        keys [
         {arriveTangent=0.3137, interpMode=RCIM_Cubic, leaveTangent=0.3137, tangentMode=RCTM_User, time=0, value=1}
         {arriveTangent=-0.5451, interpMode=RCIM_Cubic, leaveTangent=-0.5451, tangentMode=RCTM_User, time=1, value=0}
        ]
       }}})
    Index 0 Param 4=&FloatFromCurve(FloatCurve=DataInterface() {propertyValues {Curve {
        defaultValue=0
        keys [
         {arriveTangent=0.3137, interpMode=RCIM_Cubic, leaveTangent=0.3137, tangentMode=RCTM_User, time=0, value=1}
         {arriveTangent=-0.5451, interpMode=RCIM_Cubic, leaveTangent=-0.5451, tangentMode=RCTM_User, time=1, value=0}
        ]
       }}})
    Module.Index0Float0Write=true
    Module.Index0Float1Write=true
    Module.Index0Float2Write=true
    Module.Index0Float3Write=true
   )
   /SetVariables(
    Particles.Flare Color:LinearColor=&MakeLinearColorFromVectorAndFloat(
     "Vector (RGB)"=&MakeVectorFromLinearColorRGB(Color=&FNTimeOfDayColor(Night Color=<rgba 0.8021 0.8021 0.8021 1>, Morning Color=<rgba 1 1 1 1>, Day Color=<rgba 3 3 3 1>, Evening Color=<rgba 1 1 1 1>))
     "Float (Alpha)"=&FloatFromCurve(FloatCurve=DataInterface() {propertyValues {Curve {
         defaultValue=1
         keys [
          {arriveTangent=9.1715, interpMode=RCIM_Cubic, leaveTangent=9.1715, tangentMode=RCTM_User, time=0, value=1}
          {interpMode=RCIM_Cubic, leaveTangent=-1, time=0.191, value=0}
         ]
        }}})
    )
    Particles.Flare Size:Vector2f=&Vector2DFromFloat(Value=&Multiply_Float(A=0.7, B=120))
   ) {isSetParameters=true}
   /SetVariables(
    Particles.Glow Color=&Multiply_LinearColorByFloat(
     Linear Color=&MultiplyLinearColor(
      Linear Color=$Particles.Color
      Scale Factor=&MakeVector4_FromLinearColor(LinearColor=&FNTimeOfDayColor(
        Night Color=<rgba 0.9219 0.8201 0.8289 0.7>
        Morning Color=<rgba 1.2 1.0675 1.0789 0.7>
        Day Color=<rgba 2 1.7791 1.7982 0.7>
        Evening Color=<rgba 1.2 1.0675 1.0789 0.7>
       ))
     )
     Float=0.33
    )
    Particles.Glow Size=&Vector2DFromFloat(Value=&Multiply_Float(
      A=$System.HeartSize
      B=&Multiply_Float(
       A=70
       B=&FloatFromCurve(FloatCurve=DataInterface() {propertyValues {Curve {
           defaultValue=0
           keys [
            {arriveTangent=3.6099, interpMode=RCIM_Cubic, leaveTangent=3.6099, tangentMode=RCTM_User, time=0, value=0}
            {arriveTangent=0.0224, interpMode=RCIM_Cubic, leaveTangent=0.0224, tangentMode=RCTM_User, time=1, value=1}
           ]
          }}})
      )
     ))
   ) {isSetParameters=true}
   &ScaleColor(
    Scale Alpha=&FloatFromCurve(FloatCurve=DataInterface() {propertyValues {Curve {
        defaultValue=0
        keys [
         {arriveTangent=0.3137, interpMode=RCIM_Cubic, leaveTangent=0.3137, tangentMode=RCTM_User, time=0, value=1}
         {arriveTangent=-2.0958, interpMode=RCIM_Cubic, leaveTangent=-2.0958, tangentMode=RCTM_User, time=1, value=0}
        ]
       }}})
    ScaleA=true
    ScaleRGB=true
   )
  ]
  rendererClasses [&NiagaraSpriteRendererProperties]
  renderers [
   {rendererClass=&NiagaraSpriteRendererProperties}
   {rendererClass=&NiagaraSpriteRendererProperties, rendererIndex=1}
   {rendererClass=&NiagaraSpriteRendererProperties, rendererIndex=2}
  ]
 }
}
summary {
 systemName=NS_ReactFX_RoseBruin
 emitters [
  {emitterName=HeartsSparkles, rendererClasses [&NiagaraSpriteRendererProperties]}
  {emitterName=NewIntro, rendererClasses [&NiagaraSpriteRendererProperties]}
  {emitterName=NewPop, rendererClasses [&NiagaraSpriteRendererProperties]}
 ]
 userVariables [
  User.bIsFrontEnd {type=NiagaraBool}
  User.bIsFrontEndPreview {type=NiagaraBool}
 ]
}
