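# Niagara one-shot reaction effect for a cosmetic companion, showing a heart pop with spikes and embers at a shared burst position and playing a petting-reaction sound.
# Heart layers use spawn groups to switch size, timing and color, with front-end and time-of-day tints.
# Spikes burst from a hand-spaced, camera-facing ring, followed by a brief spherical spray.
# Embers follow the system intro delay, drifting with curl noise; SmokeLWE_Burst has no renderers.
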NiagaraSystem /CosmeticCompanions/FX/React/NS_ReactFX_SandyInland
dependencies {
 usedDataInterfaces [
  &NiagaraDataInterfaceCurve
  &NiagaraDataInterfaceCamera
  &NiagaraDataInterfaceColorCurve
  &NiagaraDataInterfaceVector2DCurve
  &NiagaraDataInterfaceAudioPlayer
  &NiagaraDataInterfaceSpriteRendererInfo
 ]
 usedDynamicInputs [&FN_FrontEndVector3Modifier]
 usedModules [
  &self:SystemSpawnScript.SystemScriptSource.SystemScriptGraph.NiagaraNodeAssignment_0.SetVariables_#1
  &self:SystemSpawnScript.SystemScriptSource.SystemScriptGraph.NiagaraNodeAssignment_1.SetVariables_#2
  &self:SystemSpawnScript.SystemScriptSource.SystemScriptGraph.NiagaraNodeAssignment_2.SetVariables_#3
  &SystemState
 ]
}
emitters {
 Embers {
  emitterUpdateScript [
   &EmitterState(
    Life Cycle Mode=::NewEnumerator1
    Loop Behavior=::NewEnumerator1
    Loop Delay=&Add_Float(A=$System.IntroLength, B=0.15)
    Loop Duration=0.25
    UseLoopDelay=true
   )
   &SpawnRate(SpawnRate=80, Use Spawn Probability=false)
  ]
  particleSpawnScript [
   &InitializeParticle(
    Lifetime=/Niagara/DynamicInputs/Helpers/DistributionCurveFloat(
     Minimum:NiagaraFloat=0.3
     Maximum:NiagaraFloat=1
     Distribution Curve:NiagaraDataInterfaceCurve=DataInterface() {propertyValues {Curve {defaultValue=0, keys [{tangentMode=RCTM_User, time=0, value=0}, <at 0.6583 0.1749>, {tangentMode=RCTM_User, time=1, value=1}]}}}
     Curve Parameter Type:NiagaraInt32=0
     "Curve Parameter Type:":NiagaraInt32=0
     Curve Parameter:NiagaraFloat=?
     Evaluation Type:ENiagaraRandomnessEvaluation=::NewEnumerator0
     Recalculate Random Each Loop:NiagaraBool=true
     Randomness Mode:ENiagaraRandomnessMode=::NewEnumerator0
     Random Seed:NiagaraInt32=0
     Fixed Random Seed:NiagaraBool=false
    )
    Position Offset=$System.BurstPosition
    Sprite Size Mode=::NewEnumerator1
    Uniform Sprite Size Max=1.8
    Uniform Sprite Size Min=1
    UsePositionOffset=true
    Write Material Random=true
   )
   &SubUVAnimation(
    End Frame Range Override=3
    Evaluation Type=::NewEnumerator0
    "Sprite Renderer "=DataInterface() {propertyValues {SpriteRenderer=&self:Embers_0.NiagaraSpriteRendererProperties_0}}
    Start Frame Range Override=1
    SubUV Animation Mode:ENiagara_SubUVLookupModeV2=::NewEnumerator2
    UseEndFrame=true
    UseStartFrame=true
   )
   &ShapeLocation(MaskSpawnGroup=false, Override Seed=false, Sphere Radius=15, UseOwnerScale=true)
   &CameraOffset(Camera Offset Amount=15)
   &AddVelocity(
    Cone Angle=140
    Cone Axis=<0 0 1>
    Module.UseConeFalloff=false
    Module.UseVelDistribution=false
    Override Seed=false
    Velocity Mode=::NewEnumerator2
    Velocity Speed=&RandomRangeFloat(Minimum=70, Maximum=100)
   )
  ]
  particleUpdateScript [
   &ParticleState()
   &ScaleSpriteSize(Uniform Curve Sprite Scale=DataInterface() {propertyValues {Curve {
       defaultValue=0
       keys [
        {interpMode=RCIM_Cubic, tangentMode=RCTM_User, time=0, value=0.9553}
        {interpMode=RCIM_Cubic, tangentMode=RCTM_User, time=0.5, value=1}
        {arriveTangent=-4.7965, interpMode=RCIM_Cubic, leaveTangent=-4.7965, tangentMode=RCTM_User, time=1, value=0}
       ]
      }}})
   &ScaleColor(
    Linear Color Curve=DataInterface() {propertyValues {
      AlphaCurve {keys [<at 0 0.5 RCIM_Cubic>]}
      BlueCurve {keys [<at 0.2131 0 RCIM_Cubic>,<at 0.5213 0>,<at 0.7206 0.0001>,<at 1 0.0034>]}
      GreenCurve {keys [<at 0.2131 1.7343 RCIM_Cubic>,<at 0.5213 0.2365>,<at 0.7206 0.0172>,<at 1 0>]}
      RedCurve {keys [<at 0.2131 8 RCIM_Cubic>,<at 0.5213 2>,<at 0.7206 1>,<at 1 0.3698>]}
     }}
    Scale Mode=2
   )
   /SetVariables(Particles.OutlineSize:Vector2f=&Multiply_Vector2D_ByFloat(Float=2, Vector2D=$Particles.Initial.SpriteSize)) {isSetParameters=true}
   &Drag(Drag=&RandomRangeFloat(Minimum=3, Maximum=4), Use Linear Drag=true, Use Rotational Drag=true)
   &CurlNoiseForce(Bias Noise=false, Noise Frequency=5, Noise Strength=30, Pan Noise=false)
   &SolveForcesAndVelocity(Clamp Velocity=false, Limit Acceleration=false)
  ]
  rendererClasses [&NiagaraSpriteRendererProperties]
  renderers [
   {rendererClass=&NiagaraSpriteRendererProperties}
   {rendererClass=&NiagaraSpriteRendererProperties, rendererIndex=1}
  ]
 }
 Heart {
  emitterSpawnScript [
   &InitializeAudioPlayer(Audio Player=DataInterface() {propertyValues {
      Attenuation=null
      Concurrency=null
      MaxPlaysPerTick=10
      SoundToPlay=/CosmeticCompanions/Assets/Audio_Shared/Interaction/Petting/SC_Mimosa_Interaction_React_Default
      bAllowLoopingOneShotSounds=false
      bLimitPlaysPerTick=true
      bOnlyActiveDuringGameplay=false
      bStopWhenComponentIsDestroyed=true
      ConfigurationUserParameter {parameter {name=None, typeDefHandle {registeredTypeIndex=1}, varData []}}
      ParameterNames []
     }})
   /Niagara/Modules/Audio/PlayAudio() {module=PlayAudio001}
  ]
  emitterUpdateScript [
   &EmitterState(Life Cycle Mode=::NewEnumerator1, Loop Behavior=::NewEnumerator1, Loop Duration=2, UseLoopDelay=false)
   &SpawnBurst_Instantaneous(Spawn Group=3, Spawn Time=0.9, Use Spawn Probability=false, UseLoopCountLimit=false) {module=SpawnBurst_Instantaneous003}
   &SpawnBurst_Instantaneous(Spawn Group=2, Spawn Time=0.9, Use Spawn Probability=false, UseLoopCountLimit=false) {module=SpawnBurst_Instantaneous002}
   &SpawnBurst_Instantaneous(Spawn Group=1, Use Spawn Probability=false, UseLoopCountLimit=false) {module=SpawnBurst_Instantaneous001}
   &SpawnBurst_Instantaneous(Use Spawn Probability=false, UseLoopCountLimit=false)
  ]
  particleSpawnScript [
   /SetVariables(Particles.VisibilityTag=&SpawnGroup()) {isSetParameters=true}
   &InitializeParticle(
    Color=&MergeHelper_ColorSwitch(
     0=&MultiplyLinearColor(
      Linear Color=&MakeCustomLinearColorFromBool(
       Bool=$User.bIsFrontEndPreview
       True Color=<rgba 0.5 0.5 0.5 1>
       False Color=&FNTimeOfDayColor(
        Night Color=<rgba 0.6117 0.7914 1 1>
        Morning Color=<rgba 0.8976 0.8104 0.7205 1>
        Day Color=<rgba 1 1 1 1>
        Evening Color=<rgba 0.9022 0.8146 0.7243 1>
       )
      )
      Scale Factor=&MakeVector4_FromLinearColor(LinearColor=<rgba 1 1 1 1>)
     )
     1=&MultiplyLinearColor(
      Linear Color=&MakeCustomLinearColorFromBool(
       Bool=$User.bIsFrontEndPreview
       True Color=<rgba 0.5 0.5 0.5 1>
       False Color=&FNTimeOfDayColor(
        Night Color=<rgba 0.6117 0.7914 1 1>
        Morning Color=<rgba 0.8976 0.8104 0.7205 1>
        Day Color=<rgba 1 1 1 1>
        Evening Color=<rgba 0.9022 0.8146 0.7243 1>
       )
      )
      Scale Factor=&MakeVector4_FromLinearColor(LinearColor=<rgba 1 1 1 1>)
     )
     2=<rgba 1 1 1 1>
     3=&MakeCustomLinearColorFromBool(
      Bool=&Boolean_OrOperation(Bool1=$User.bIsFrontEnd, Bool2=$User.bIsFrontEndPreview)
      True Color=<rgba 0.7513 0.8809 1 0.1>
      False Color=<rgba 1 1 1 0.5>
     )
     Index=&SpawnGroup()
     Num=::NewEnumerator2
    )
    Lifetime=&MergeHelper_FloatSwitch(0=1.5, 1=1.5, 2=0.4, 3=0.4, Index=&SpawnGroup(), Num=::NewEnumerator2)
    Position Offset=&MergeHelper_VectorSwitch(
     0=$System.BurstPosition
     1=$System.BurstPosition
     2=&Add_Vector(A=$System.BurstPosition, B=<0 0 2>)
     3=&Add_Vector(A=$System.BurstPosition, B=<0 0 2>)
     Index=&SpawnGroup()
     Num=::NewEnumerator2
    )
    Position Offset Coordinate Space=::Local
    Sprite Size Mode=::NewEnumerator3
    Uniform Sprite Size=&MergeHelper_FloatSwitch(
     0=25
     1=25
     2=&Multiply_Float(A=35, B=1.0594)
     3=&Multiply_Float(A=35, B=1.0594)
     Index=&SpawnGroup()
     Num=::NewEnumerator2
    )
    UsePositionOffset=true
    Write Material Random=true
   )
   /SetVariables(Particles.RightRootSize:Vector2f=<35 35>) {isSetParameters=true}
   &CameraOffset(Camera Offset Amount=&MergeHelper_FloatSwitch(0=10, 1=9.5, 2=11, 3=11, Index=&SpawnGroup(), Num=::NewEnumerator2))
  ]
  particleUpdateScript [
   &ParticleState()
   &ScaleSpriteSize(
    Scale Sprite Size Mode=::NewEnumerator0
    Uniform Scale Factor=&MergeHelper_FloatSwitch(
     0=&FloatFromCurve(
      FloatCurve=DataInterface() {propertyValues {Curve {
         defaultValue=0
         keys [
          {arriveTangent=3.7864, interpMode=RCIM_Cubic, leaveTangent=3.7864, tangentMode=RCTM_User, time=0, value=0}
          {arriveTangent=3.5868, interpMode=RCIM_Cubic, leaveTangent=3.5868, time=0.0786, value=0.7773}
          {interpMode=RCIM_Cubic, leaveTangent=0.543, time=0.2954, value=1.0594}
         ]
        }}}
      CurveIndex=$Particles.Age
     )
     1=&FloatFromCurve(
      FloatCurve=DataInterface() {propertyValues {Curve {
         defaultValue=0
         keys [
          {arriveTangent=3.7864, interpMode=RCIM_Cubic, leaveTangent=3.7864, tangentMode=RCTM_User, time=0.1079, value=0}
          {arriveTangentWeight=0.193, interpMode=RCIM_Cubic, leaveTangentWeight=0.1615, tangentMode=RCTM_User, time=0.163, value=1.05}
          {arriveTangent=-0.4312, interpMode=RCIM_Cubic, leaveTangent=-0.4312, time=0.1944, value=1.0831}
          {interpMode=RCIM_Cubic, leaveTangent=0.543, time=0.2789, value=1}
         ]
        }}}
      CurveIndex=$Particles.Age
     )
     2=1
     3=1
     Index=&SpawnGroup()
     Num=::NewEnumerator2
    )
   )
   &ScaleColor(
    Scale Alpha=&MergeHelper_FloatSwitch(
     0=&FloatFromCurve(FloatCurve=DataInterface() {propertyValues {Curve {keys [<at 0 1>,<at 0.2765 1 RCIM_Constant>,<at 0.7162 0>]}}})
     1=&FloatFromCurve(FloatCurve=DataInterface() {propertyValues {Curve {keys [<at 0 1>,<at 0.2765 1 RCIM_Constant>,<at 0.7162 0>]}}})
     2=1
     3=1
     Index=&SpawnGroup()
     Num=::NewEnumerator2
    )
    Scale RGB=&MakeVectorFromLinearColorRGB(Color=&MergeHelper_ColorSwitch(
      0=&ColorFromCurve(ColorCurve=DataInterface() {propertyValues {
         AlphaCurve {keys [<at 0 1 RCIM_Cubic>]}
         BlueCurve {keys [<at 0.6994 1 RCIM_Cubic>,<at 0.7384 0.5122 RCIM_Cubic>]}
         GreenCurve {keys [<at 0.6994 1 RCIM_Cubic>,<at 0.7384 0 RCIM_Cubic>]}
         RedCurve {keys [<at 0.6994 1 RCIM_Cubic>,<at 0.7384 0.3698 RCIM_Cubic>]}
        }})
      1=&ColorFromCurve(ColorCurve=DataInterface() {propertyValues {
         AlphaCurve {keys [<at 0 1 RCIM_Cubic>]}
         BlueCurve {keys [<at 0.6994 1 RCIM_Cubic>,<at 0.7384 0.5122 RCIM_Cubic>]}
         GreenCurve {keys [<at 0.6994 1 RCIM_Cubic>,<at 0.7384 0 RCIM_Cubic>]}
         RedCurve {keys [<at 0.6994 1 RCIM_Cubic>,<at 0.7384 0.3698 RCIM_Cubic>]}
        }})
      2=<rgba 1 1 1 1>
      3=<rgba 1 1 1 1>
      Index=&SpawnGroup()
      Num=::NewEnumerator2
     ))
    ScaleA=true
    ScaleRGB=true
   )
   &SubUVAnimation(
    End Frame Range Override=11
    "Sprite Renderer "=DataInterface() {propertyValues {SpriteRenderer=&self:Heart_0.NiagaraSpriteRendererProperties_3}}
    UseEndFrame=true
    UseStartFrame=true
   )
   /SetVariables() {isSetParameters=true}
  ]
  rendererClasses [&NiagaraSpriteRendererProperties]
  renderers [
   {rendererClass=&NiagaraSpriteRendererProperties}
   {rendererClass=&NiagaraSpriteRendererProperties, rendererIndex=1}
   {rendererClass=&NiagaraSpriteRendererProperties, rendererIndex=2}
   {rendererClass=&NiagaraSpriteRendererProperties, rendererIndex=3}
  ]
 }
 SmokeLWE_Burst {rendererClasses [], renderers []}
 Spikes {
  emitterUpdateScript [
   &EmitterState(Life Cycle Mode=::NewEnumerator1, Loop Behavior=::NewEnumerator1, Loop Duration=1, UseLoopDelay=false)
   &SpawnBurst_Instantaneous(Spawn Count=10, Use Spawn Probability=false, UseLoopCountLimit=false)
   &SpawnRate(
    SpawnGroup=1
    SpawnRate=&FloatFromCurve(
     FloatCurve=DataInterface() {propertyValues {Curve {keys [<at 0.0642 0 RCIM_Constant>,<at 0.15 1 RCIM_Constant>,<at 0.35 0 RCIM_Constant>]}}}
     CurveIndex=$Emitter.Age
     Scale Curve=40
    )
    Use Spawn Probability=false
   )
  ]
  particleSpawnScript [
   /SetVariables(Particles.VisibilityTag=&SpawnGroup()) {isSetParameters=true}
   &InitializeParticle(
    Color=&FNTimeOfDayColor(
     Night Color=<rgba 0.4832 0.7459 1 1>
     Morning Color=<rgba 1 0.8695 0.7944 1>
     Day Color=<rgba 1 1 1 1>
     Evening Color=<rgba 1 0.8695 0.7944 1>
    )
    Lifetime=&MaskFloat(
     Mask Source=0
     Mask Parameter=?
     Mask=0
     Masked Float=&RandomRangeFloat(Minimum=0.2, Maximum=0.4)
     Passthrough Float=&RandomRangeFloat(Minimum=0.3, Maximum=0.6)
    )
    Position Offset=$System.BurstPosition
    Sprite Rotation Angle=&MaskFloat(
     Mask Source=0
     Mask Parameter=?
     Mask=0
     Masked Float=0
     Passthrough Float=&RandomRangeFloat(Minimum=0, Maximum=1)
    )
    Sprite Rotation Mode=::NewEnumerator1
    Sprite Size=/Niagara/DynamicInputs/Helpers/MaskVector2d(
     Mask Source:NiagaraInt32=0
     Mask Parameter:NiagaraInt32=?
     Mask:NiagaraInt32=0
     Masked Vector2D:Vector2f=&Multiply_Vector2D_ByFloat(Float=1, Vector2D=&RandomRangeVector2D(Minimum=<18 14>, Maximum=<38 30>))
     Passthrough Vector2D:Vector2f=&RandomRangeVector2D(Minimum=<18 18>, Maximum=<38 38>)
    )
    Sprite Size Mode=::NewEnumerator4
    UsePositionOffset=true
    Write Material Random=true
   )
   &ShapeLocation(
    Additional Quaternion Rotation=true
    Additional Yaw / Pitch / Roll=<0 0.25 0>
    Handle Radius=1
    Large Radius=4
    MaskSpawnGroup=true
    Rotation Mode=::NewEnumerator3
    Rotation Quaternion=&OrientMeshToVector(
     Look At Direction=&ReturnCameraVector(Camera Vectors=::NewEnumerator1, Coordinate Space=::Simulation)
     Reference Axis Vector=<1 0 0>
    )
    Shape Primitive=::NewEnumerator3
    Torus Distribution Mode:ENiagaraTorusDistributionMode=::NewEnumerator1
    U Position=&MergeHelper_FloatSwitch(
     0=0
     1=0.1444
     2=0.2711
     3=0.3444
     4=0.4356
     5=0.4644
     6=0.5378
     7=0.6556
     8=0.7733
     9=0.8467
     Index=&ReturnExecIndex()
     Num=::NewEnumerator8
    )
    UseOwnerScale=true
   )
   &ShapeLocation(
    Additional Quaternion Rotation=true
    Additional Yaw / Pitch / Roll=<0 0.25 0>
    MaskSpawnGroup=true
    Offset=<0 0 2>
    Offset Mode=::NewEnumerator0
    Override Seed=false
    Rotation Mode=::NewEnumerator3
    Rotation Quaternion=&OrientMeshToVector(
     Look At Direction=&ReturnCameraVector(
      Camera Vectors=::NewEnumerator1
      Coordinate Space=::Simulation
      CameraQuery=DataInterface() {propertyValues {bRequireCurrentFrameData=false}}
     )
     Reference Axis Vector=<1 0 0>
    )
    Spawn Group Mask=1
    Sphere Radius=19
    Sphere Surface Distribution=1
    UseOwnerScale=true
   ) {module=ShapeLocation001}
   &AddVelocity(
    Constrain To Radius=100
    Module.Use Falloff Distance=false
    Origin Offset=&MaskVector(Mask Source=0, Mask Parameter=?, Mask=0, Masked Vector=<0 0 0>, Passthrough Vector=<0 0 1>)
    Override Seed=false
    Velocity Mode=::NewEnumerator1
    Velocity Origin:Vector3f=$System.BurstPosition
    Velocity Speed=&MaskFloat(
     Mask Source=0
     Mask Parameter=?
     Mask=0
     Masked Float=&RandomRangeFloat(Minimum=80, Maximum=120)
     Passthrough Float=&RandomRangeFloat(Minimum=10, Maximum=20)
    )
   )
   &CameraOffset(Camera Offset Amount=15)
   /SetVariables(Particles.Mote Sprite Size:Vector2f=&Vector2DFromFloat(Value=&RandomRangeFloat(Minimum=3, Maximum=6))) {isSetParameters=true}
  ]
  particleUpdateScript [
   &ParticleState()
   /SetVariables(Particles.Flicker:NiagaraFloat=&Waveform(
     "Waveform [1]"=::NewEnumerator0
     "[1] Frequency"=&RandomRangeFloat(Minimum=3, Maximum=5)
     "[1] Amplitude Min/Max"=<0 1>
    )) {isSetParameters=true}
   &Drag(
    Drag=&Multiply_Float(
     A=&RandomRangeFloat(Minimum=3, Maximum=7)
     B=&FloatFromCurve(
      FloatCurve=DataInterface() {propertyValues {Curve {defaultValue=1, keys [{tangentMode=RCTM_User, time=0.1973, value=0}, {tangentMode=RCTM_User, time=0.4549, value=2}]}}}
      Scale Curve=&MaskFloat(
       Mask Source=0
       Mask Parameter=?
       Mask=0
       Masked Float=&RandomRangeFloat(Minimum=0.5, Maximum=1)
       Passthrough Float=1
      )
     )
    )
    Use Linear Drag=true
    Use Rotational Drag=true
   )
   &SolveForcesAndVelocity(Clamp Velocity=false, Limit Acceleration=false)
   &ScaleColor(
    Scale Alpha=&FloatFromCurve(FloatCurve=DataInterface() {propertyValues {Curve {
        defaultValue=1
        keys [
         {interpMode=RCIM_Cubic, tangentMode=RCTM_User, time=0.2721, value=1}
         {arriveTangent=-2.8008, interpMode=RCIM_Cubic, leaveTangent=-2.8008, tangentMode=RCTM_User, time=0.7239, value=0}
        ]
       }}})
    Scale RGB=&MakeVectorFromLinearColorRGB(Color=&ColorFromCurve(ColorCurve=DataInterface() {propertyValues {
        AlphaCurve {keys []}
        BlueCurve {keys [<at 0 2 RCIM_Cubic>]}
        GreenCurve {keys [<at 0 0.3424 RCIM_Cubic>]}
        RedCurve {keys [<at 0 0.4809 RCIM_Cubic>]}
       }}))
    ScaleA=true
    ScaleRGB=true
   )
   /SetVariables(Particles.Hearts SpriteSize:Vector2f=&Multiply_Vector2D(
     A=&MaskVector2d(
      Mask Source:NiagaraInt32=0
      Mask Parameter:NiagaraInt32=?
      Mask:NiagaraInt32=0
      Masked Vector2D:Vector2f=&Vector2DFromCurve(
       Vector2Curve=DataInterface() {propertyValues {
         XCurve {keys [<at 0.1 0.2>,<at 0.2493 0.2>,<at 0.5 0.8>,<at 0.6847 0.8>,<at 0.9953 0.2>]}
         YCurve {keys [<at 0.1 1>,<at 0.2471 1.4662>,<at 0.5 1>]}
        }}
       ScaleCurve=<1 1>
      )
      Passthrough Vector2D:Vector2f=&Vector2DFromCurve(
       Vector2Curve=DataInterface() {propertyValues {
         XCurve {keys [<at 0.1 0.6751>,<at 0.2979 1.147>,<at 0.5 0.9218>,<at 0.8505 0.9218>,<at 0.9953 0.2>]}
         YCurve {keys [<at 0.1 1>,<at 0.2448 1.3223>,<at 0.5 1>]}
        }}
       ScaleCurve=<1 1>
      )
     )
     B=&Vector2DFromFloat(Value=&FloatFromCurve(
       FloatCurve=DataInterface() {propertyValues {Curve {
          defaultValue=1
          keys [
           {arriveTangent=3.2566, interpMode=RCIM_Cubic, leaveTangent=3.2566, tangentMode=RCTM_User, time=0, value=0.5}
           {arriveTangent=-0.5, interpMode=RCIM_Cubic, leaveTangent=-0.5, time=0.5039, value=1.1092}
           {arriveTangent=0.0224, interpMode=RCIM_Cubic, leaveTangent=0.0224, tangentMode=RCTM_User, time=1, value=0}
          ]
         }}}
       Scale Curve=&Multiply_Float(A=1, B=&RandomRangeFloat(Minimum=8, Maximum=3))
      ))
    )) {isSetParameters=true}
   /SetVariables(
    Particles.Smoke Dynam 1:Vector4f=&MakeVector4(
     X=&Add_Float(
      A=&RandomRangeFloat(Minimum=0, Maximum=1)
      B=&FloatFromCurve(
       FloatCurve=DataInterface() {propertyValues {Curve {keys [
           <at 0 0 RCIM_Cubic>
           {arriveTangent=1.5947, interpMode=RCIM_Cubic, leaveTangent=1.5947, tangentMode=RCTM_User, time=0.2557, value=0.5257}
           <at 1 1 RCIM_Cubic>
          ]}}}
       CurveIndex=$Particles.Age
       Scale Curve=0.25
      )
     )
     Y=0
     Z=1
     W=1
    )
    Particles.Smoke Dynam 2:Vector4f=&MakeVector4(X=0, Y=0, Z=0, W=0)
    Particles.Smoke Dynam 3:Vector4f=&MakeVector4(X=0, Y=0, Z=2, W=0)
    Particles.Smoke Dynam 4:Vector4f=&MakeVector4(X=1, Y=70, Z=0, W=0)
   ) {isSetParameters=true}
   /SetVariables(Particles.Mote Sprite Size:Vector2f=&Multiply_Vector2D(
     A=$Particles.Initial.Mote Sprite Size
     B=&Vector2DFromCurve(
      Vector2Curve=DataInterface() {propertyValues {
        XCurve {keys [
          {arriveTangent=0.1643, interpMode=RCIM_Cubic, leaveTangent=0.1643, leaveTangentWeight=0.7988, tangentMode=RCTM_User, time=0, value=0.4}
          <at 1 1 RCIM_Cubic>
         ]}
        YCurve {keys [
          {arriveTangent=7.1339, interpMode=RCIM_Cubic, leaveTangent=7.1339, leaveTangentWeight=1.1393, tangentMode=RCTM_User, time=0, value=0.4}
          {arriveTangent=-5.7403, interpMode=RCIM_Cubic, leaveTangent=-5.7403, tangentMode=RCTM_User, time=0.794, value=2}
          <at 1 1.5 RCIM_Cubic>
         ]}
       }}
      ScaleCurve=&Multiply_Vector2D_ByFloat(Float=0.3671, Vector2D=<1 1>)
     )
    )) {isSetParameters=true}
   /SetVariables(Particles.GlowColor=&MakeLinearColorFromVectorAndFloat(
     "Vector (RGB)"=&MakeVectorFromLinearColorRGB(Color=&ColorFromCurve(ColorCurve=DataInterface() {propertyValues {
         AlphaCurve {keys []}
         BlueCurve {keys [<at 0 1 RCIM_Cubic>]}
         GreenCurve {keys [<at 0 0.1712 RCIM_Cubic>]}
         RedCurve {keys [<at 0 0.2404 RCIM_Cubic>]}
        }}))
     "Float (Alpha)"=&FloatFromCurve(
      FloatCurve=DataInterface() {propertyValues {Curve {
         defaultValue=1
         keys [
          {interpMode=RCIM_Cubic, tangentMode=RCTM_User, time=0.2721, value=1}
          {arriveTangent=-2.8008, interpMode=RCIM_Cubic, leaveTangent=-2.8008, tangentMode=RCTM_User, time=1, value=0}
         ]
        }}}
      Scale Curve=0.5
     )
    )) {isSetParameters=true}
   /SetVariables(
    Particles.OutlineColor:LinearColor=&MakeLinearColorFromVectorAndFloat(
     "Vector (RGB)"=<0 0 0>
     "Float (Alpha)"=&MaskFloat(
      Mask Source=0
      Mask Parameter=?
      Mask=0
      Masked Float=&FloatFromCurve(FloatCurve=DataInterface() {propertyValues {Curve {
          defaultValue=1
          keys [
           {interpMode=RCIM_Cubic, tangentMode=RCTM_User, time=0.2721, value=1}
           {arriveTangent=-2.8008, interpMode=RCIM_Cubic, leaveTangent=-2.8008, tangentMode=RCTM_User, time=1, value=0}
          ]
         }}})
      Passthrough Float=&FloatFromCurve(
       FloatCurve=DataInterface() {propertyValues {Curve {
          defaultValue=1
          keys [
           {arriveTangent=-2.463, interpMode=RCIM_Cubic, leaveTangent=-2.463, tangentMode=RCTM_User, time=0.3842, value=1}
           {arriveTangent=-0.1815, interpMode=RCIM_Cubic, leaveTangent=-0.1815, tangentMode=RCTM_User, time=0.7919, value=0}
          ]
         }}}
       Scale Curve=10
      )
     )
    )
    Particles.OutlineSize:Vector2f=&Vector2DFromFloat(Value=/Niagara/DynamicInputs/TypeConversions/Vector2DLength(Vector:Vector2f=$Particles.Hearts SpriteSize))
   ) {isSetParameters=true}
  ]
  rendererClasses [&NiagaraSpriteRendererProperties]
  renderers [
   {rendererClass=&NiagaraSpriteRendererProperties}
   {rendererClass=&NiagaraSpriteRendererProperties, rendererIndex=1}
   {rendererClass=&NiagaraSpriteRendererProperties, rendererIndex=2}
   {rendererClass=&NiagaraSpriteRendererProperties, rendererIndex=3}
  ]
 }
}
summary {
 systemName=NS_ReactFX_SandyInland
 emitters [
  {emitterName=Spikes, rendererClasses [&NiagaraSpriteRendererProperties]}
  {emitterName=Heart, rendererClasses [&NiagaraSpriteRendererProperties]}
  {emitterName=SmokeLWE_Burst, rendererClasses []}
  {emitterName=Embers, rendererClasses [&NiagaraSpriteRendererProperties]}
 ]
 userVariables [
  User.bIsFrontEnd {type=NiagaraBool}
  User.bIsFrontEndPreview {type=NiagaraBool}
 ]
}
