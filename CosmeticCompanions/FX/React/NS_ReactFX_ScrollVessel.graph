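# Cosmetic companion reaction effect for ScrollVessel: a pulsing sprite burst with stars and the shared petting-reaction sound.
# Intro plays the audio and spawns two triple-rendered sprites, with per-partition scale curves, glow bindings and animated material parameters.
# Near the intro's end, GPU Stars spawn three sprites on camera-relative axes, later kicked outward against curve-driven gravity.
# System variables share timing, burst position and color; three LWE emitters have no renderers.
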NiagaraSystem /CosmeticCompanions/FX/React/NS_ReactFX_ScrollVessel
dependencies {
 usedDataInterfaces [&NiagaraDataInterfaceAudioPlayer, &NiagaraDataInterfaceCurve, &NiagaraDataInterfaceCamera]
 usedDynamicInputs [/Niagara/DynamicInputs/TypeConversions/MakeCustomPositionFromBool, &FN_FrontEndVector3Modifier]
 usedModules [
  &self:SystemSpawnScript.SystemScriptSource.SystemScriptGraph.NiagaraNodeAssignment_0.SetVariables_#1
  &self:SystemSpawnScript.SystemScriptSource.SystemScriptGraph.NiagaraNodeAssignment_1.SetVariables_#2
  &self:SystemSpawnScript.SystemScriptSource.SystemScriptGraph.NiagaraNodeAssignment_3.SetVariables_#3
  &self:SystemSpawnScript.SystemScriptSource.SystemScriptGraph.NiagaraNodeAssignment_4.SetVariables_#4
  &self:SystemSpawnScript.SystemScriptSource.SystemScriptGraph.NiagaraNodeAssignment_7.SetVariables_#5
  &self:SystemSpawnScript.SystemScriptSource.SystemScriptGraph.NiagaraNodeAssignment_9.SetVariables_#6
  &self:SystemSpawnScript.SystemScriptSource.SystemScriptGraph.NiagaraNodeAssignment_10.SetVariables_#7
  &self:SystemSpawnScript.SystemScriptSource.SystemScriptGraph.NiagaraNodeAssignment_5.SetVariables_#8
  &self:SystemSpawnScript.SystemScriptSource.SystemScriptGraph.NiagaraNodeAssignment_6.SetVariables_#9
  &SystemState
  &self:SystemSpawnScript.SystemScriptSource.SystemScriptGraph.NiagaraNodeAssignment_2.SetVariables_#10
 ]
}
emitters {
 Intro {
  emitterSpawnScript [
   &InitializeAudioPlayer(Audio Player=DataInterface() {propertyValues {
      Attenuation=null
      Concurrency=null
      MaxPlaysPerTick=10
      SoundToPlay=/CosmeticCompanions/Assets/Audio_Shared/Interaction/Petting/SC_Mimosa_Interaction_React_Default
      bAllowLoopingOneShotSounds=false
      bLimitPlaysPerTick=true
      bOnlyActiveDuringGameplay=false
      bStopWhenComponentIsDestroyed=true
      ConfigurationUserParameter {parameter {name=None, typeDefHandle {registeredTypeIndex=1}, varData []}}
      ParameterNames []
     }})
   /Niagara/Modules/Audio/PlayAudio() {module=PlayAudio001}
  ]
  emitterUpdateScript [
   &EmitterState()
   &SpawnBurst_Instantaneous(Spawn Count=2, Use Spawn Probability=false, UseLoopCountLimit=false)
  ]
  particleSpawnScript [
   &PartitionParticles()
   &InitializeParticle(
    Lifetime=$System.IntroLength
    Sprite Rotation Mode=::NewEnumerator0
    Sprite Size Mode=::NewEnumerator3
    Uniform Sprite Size=29
    UsePositionOffset=false
    Write Material Random=true
   )
  ]
  particleUpdateScript [
   &ParticleState()
   &ScaleSpriteSize(
    Scale Sprite Size Mode=::NewEnumerator0
    Uniform Scale Factor=&Add_Float(
     A=0.5
     B=&MergeHelper_FloatSwitch(
      0=&FloatFromCurve(FloatCurve=DataInterface() {propertyValues {Curve {
          defaultValue=0
          keys [
           {arriveTangent=2.8117, interpMode=RCIM_Cubic, leaveTangent=2.8117, tangentMode=RCTM_User, time=0.0035, value=0}
           {arriveTangent=24.237, interpMode=RCIM_Cubic, leaveTangent=24.237, tangentMode=RCTM_User, time=0.0726, value=0.7902}
           {
            arriveTangent=-9.6484
            arriveTangentWeight=0.2036
            interpMode=RCIM_Cubic
            leaveTangent=-9.6484
            leaveTangentWeight=0.2153
            tangentMode=RCTM_User
            time=0.1617
            value=1.1488
           }
           {arriveTangent=-0.3419, interpMode=RCIM_Cubic, leaveTangent=-0.3419, tangentMode=RCTM_User, time=0.2612, value=0.5964}
           {
            arriveTangent=18.8498
            arriveTangentWeight=0.4477
            interpMode=RCIM_Cubic
            leaveTangent=18.8498
            leaveTangentWeight=0.2947
            tangentMode=RCTM_User
            time=0.3075
            value=1.1523
           }
           {
            arriveTangent=-11.2963
            arriveTangentWeight=0.0789
            interpMode=RCIM_Cubic
            leaveTangent=-11.2963
            leaveTangentWeight=0.0623
            tangentMode=RCTM_User
            time=0.404
            value=1.5418
           }
           {arriveTangent=-1.2478, interpMode=RCIM_Cubic, leaveTangent=-1.2478, tangentMode=RCTM_User, time=0.4589, value=1.3503}
           {
            arriveTangent=-0.489
            arriveTangentWeight=0.0471
            interpMode=RCIM_Cubic
            leaveTangent=-0.489
            tangentMode=RCTM_User
            time=0.8798
            value=1.1345
           }
           {arriveTangent=16.0695, interpMode=RCIM_Cubic, leaveTangent=16.0695, tangentMode=RCTM_User, time=0.9163, value=1.7398}
           {arriveTangent=-25.3711, interpMode=RCIM_Cubic, leaveTangent=-25.3711, tangentMode=RCTM_User, time=0.9484, value=1.6889}
           {arriveTangent=-0.3666, interpMode=RCIM_Cubic, leaveTangent=-0.3666, tangentMode=RCTM_User, time=1, value=0}
          ]
         }}})
      1=&FloatFromCurve(
       FloatCurve=DataInterface() {propertyValues {Curve {
          defaultValue=0
          keys [
           {arriveTangent=3.2329, interpMode=RCIM_Cubic, leaveTangent=3.2329, tangentMode=RCTM_User, time=0.0035, value=0}
           {arriveTangent=6.9044, interpMode=RCIM_Cubic, leaveTangent=6.9044, tangentMode=RCTM_User, time=0.0726, value=0.7902}
           {
            arriveTangent=-8.3351
            arriveTangentWeight=0.2036
            interpMode=RCIM_Cubic
            leaveTangent=-8.3351
            leaveTangentWeight=0.2153
            tangentMode=RCTM_User
            time=0.1768
            value=1.1352
           }
           {arriveTangent=0.048, interpMode=RCIM_Cubic, leaveTangent=0.048, time=0.2489, value=0.6604}
           {
            arriveTangent=19.0995
            arriveTangentWeight=0.4477
            interpMode=RCIM_Cubic
            leaveTangent=19.0995
            leaveTangentWeight=0.2947
            tangentMode=RCTM_User
            time=0.3216
            value=1.1421
           }
           {arriveTangent=2.5303, interpMode=RCIM_Cubic, leaveTangent=2.5303, time=0.3823, value=1.7927}
           {
            arriveTangent=-6.777
            arriveTangentWeight=0.0789
            interpMode=RCIM_Cubic
            leaveTangent=-6.777
            leaveTangentWeight=0.0623
            tangentMode=RCTM_User
            time=0.4379
            value=1.4363
           }
           {arriveTangent=-1.2014, interpMode=RCIM_Cubic, leaveTangent=-1.2014, time=0.5033, value=1.2286}
           {
            arriveTangent=-0.1202
            arriveTangentWeight=0.0471
            interpMode=RCIM_Cubic
            leaveTangent=-0.1202
            tangentMode=RCTM_User
            time=0.8694
            value=0.9179
           }
           {arriveTangent=16.5037, interpMode=RCIM_Cubic, leaveTangent=16.5037, tangentMode=RCTM_User, time=0.8955, value=1.4081}
           {arriveTangent=-13.4804, interpMode=RCIM_Cubic, leaveTangent=-13.4804, time=0.9484, value=1.6889}
           {interpMode=RCIM_Cubic, leaveTangent=-24.2221, time=1, value=0}
          ]
         }}}
       Scale Curve=0.5
      )
      2=?
      3=?
      Index=$Particles.PartitionParticles.Partition
      Num=::NewEnumerator0
     )
    )
   ) {module=ScaleSpriteSize002}
   &ScaleColor(
    Scale Alpha=&FloatFromCurve(FloatCurve=DataInterface() {propertyValues {Curve {
        defaultValue=0
        keys [
         {arriveTangent=9.0725, interpMode=RCIM_Cubic, leaveTangent=9.0725, tangentMode=RCTM_User, time=0, value=0}
         {arriveTangent=-0.2441, leaveTangent=-0.2441, time=0.1408, value=1}
         {arriveTangent=-1.146, leaveTangent=-1.146, time=0.95, value=1}
         {arriveTangent=-0.0405, interpMode=RCIM_Cubic, leaveTangent=-0.0405, tangentMode=RCTM_User, time=1, value=0}
        ]
       }}})
    ScaleA=true
    ScaleRGB=true
   )
   &OffsetPosition(Position Offset=$System.BurstPosition)
   &DynamicMaterialParameters(
    Index 0 Param 2=1
    Module.Index0Float0Write=true
    Module.Index0Float1Write=true
    Module.Index0Float2Write=true
    Module.Index0Float3Write=true
   )
   /SetVariables(Particles.DMP_Lion_00:Vector4f=&MakeVector4(
     X=&FloatFromCurve(
      FloatCurve=DataInterface() {propertyValues {Curve {
         defaultValue=1
         keys [
          {arriveTangent=31.49, interpMode=RCIM_Cubic, leaveTangent=31.49, tangentMode=RCTM_User, time=0.4, value=0}
          {arriveTangent=2.1661, interpMode=RCIM_Cubic, leaveTangent=2.1661, tangentMode=RCTM_User, time=0.5, value=0.8411}
          {arriveTangent=0.0224, interpMode=RCIM_Cubic, leaveTangent=0.0224, tangentMode=RCTM_User, time=1, value=1}
         ]
        }}}
      CurveIndex=$Particles.Age
     )
     Y=&FloatFromCurve(
      FloatCurve=DataInterface() {propertyValues {Curve {
         defaultValue=1
         keys [
          {arriveTangent=1.4996, interpMode=RCIM_Cubic, leaveTangent=1.4996, tangentMode=RCTM_User, time=0.4, value=1}
          {interpMode=RCIM_Cubic, tangentMode=RCTM_User, time=0.5, value=0}
         ]
        }}}
      CurveIndex=$Particles.Age
     )
     Z=0
     W=1
    )) {isSetParameters=true}
   /SetVariables(
    Particles.Glow Color=&Multiply_LinearColorByFloat(Linear Color=$System.Layer 2 Color, Float=0.3)
    Particles.Glow Size=&Multiply_Vector2D(A=$Particles.SpriteSize, B=<0.6 0.6>)
   ) {isSetParameters=true}
  ]
  rendererClasses [&NiagaraSpriteRendererProperties]
  renderers [
   {rendererClass=&NiagaraSpriteRendererProperties}
   {rendererClass=&NiagaraSpriteRendererProperties, rendererIndex=1}
   {rendererClass=&NiagaraSpriteRendererProperties, rendererIndex=2}
  ]
 }
 LWE_AirGlow {rendererClasses [], renderers []}
 LWE_Corona {rendererClasses [], renderers []}
 LWE_Rays {rendererClasses [], renderers []}
 Stars {
  simTarget=GPUComputeSim
  emitterUpdateScript [
   &EmitterState()
   &SpawnBurst_Instantaneous(
    Spawn Count=3
    Spawn Time=&Subtract_Float(A=$System.IntroLength, B=1)
    Use Spawn Probability=false
    UseLoopCountLimit=false
   )
  ]
  particleSpawnScript [
   /SetVariables(Particles.CameraAxis:Vector3f=&MergeHelper_VectorSwitch(
     0=&Multiply_VectorByFloat(
      Vector=&NormalizeVector(
       Vector To Normalize=&Multiply_VectorByFloat(Vector=&ReturnCameraVector(Camera Vectors=::NewEnumerator2, Coordinate Space=::Local), Float=7.5)
       Safe Normalize=true
       Safe Fallback Threshold=0
       Safe Fallback Vector=<0 0 0.001>
      )
      Float=12.1751
     )
     1=&Multiply_VectorByFloat(
      Vector=&NormalizeVector(
       Vector To Normalize=&Add_Vector(
        A=&Multiply_VectorByFloat(Vector=&ReturnCameraVector(Camera Vectors=::NewEnumerator3, Coordinate Space=::Local), Float=6)
        B=&Multiply_VectorByFloat(Vector=&ReturnCameraVector(Camera Vectors=::NewEnumerator2, Coordinate Space=::Local), Float=11)
       )
       Safe Normalize=true
       Safe Fallback Threshold=0
       Safe Fallback Vector=<0 0 0.001>
      )
      Float=12.5
     )
     2=&Multiply_VectorByFloat(
      Vector=&NormalizeVector(
       Vector To Normalize=&Add_Vector(
        A=&Multiply_VectorByFloat(Vector=&ReturnCameraVector(Camera Vectors=::NewEnumerator3, Coordinate Space=::Local), Float=-6)
        B=&Multiply_VectorByFloat(Vector=&ReturnCameraVector(Camera Vectors=::NewEnumerator2, Coordinate Space=::Local), Float=11)
       )
       Safe Normalize=true
       Safe Fallback Threshold=0
       Safe Fallback Vector=<0 0 0.001>
      )
      Float=12.5
     )
     3=?
     Index=&ReturnExecIndex()
     Num=::NewEnumerator1
    )) {isSetParameters=true}
   &InitializeParticle(
    Lifetime=$System.IntroLength
    Position Offset=&AddVectorToPosition(
     Position=Linked() {linkedVariable {name=System.BurstPosition, type=Vector3f}}
     Vector=&Multiply_VectorByFloat(Vector=$Particles.CameraAxis, Float=1.6)
    )
    Sprite Rotation Mode=::NewEnumerator0
    Sprite Size Mode=::NewEnumerator3
    Uniform Sprite Size=&MergeHelper_FloatSwitch(0=4.5, 1=3, 2=3, 3=?, Index=&ReturnExecIndex(), Num=::NewEnumerator1)
    UsePositionOffset=true
    Write Material Random=true
   )
   &DynamicMaterialParameters(
    Index 0 Param 3=1
    Module.Index0Float0Write=true
    Module.Index0Float1Write=true
    Module.Index0Float2Write=true
    Module.Index0Float3Write=true
   )
   &AddVelocity(Velocity=$Particles.CameraAxis, Velocity Speed Scale=0.1)
   &SpriteFacingAndAlignment(
    Sprite Alignment=&MergeHelper_VectorSwitch(
     0=&Multiply_VectorByFloat(
      Vector=&NormalizeVector(
       Vector To Normalize=&Multiply_VectorByFloat(Vector=&ReturnCameraVector(Camera Vectors=::NewEnumerator2, Coordinate Space=::Local), Float=7.5)
       Safe Normalize=true
       Safe Fallback Threshold=0
       Safe Fallback Vector=<0 0 0.001>
      )
      Float=12.1751
     )
     1=&Multiply_VectorByFloat(
      Vector=&NormalizeVector(
       Vector To Normalize=&Add_Vector(
        A=&Add_Vector(
         A=&Multiply_VectorByFloat(Vector=&ReturnCameraVector(Camera Vectors=::NewEnumerator3, Coordinate Space=::Local), Float=1)
         B=&Multiply_VectorByFloat(Vector=&ReturnCameraVector(Camera Vectors=::NewEnumerator2, Coordinate Space=::Local), Float=0)
        )
        B=&Multiply_VectorByFloat(Vector=&ReturnCameraVector(Camera Vectors=::NewEnumerator1, Coordinate Space=::Local), Float=-1)
       )
       Safe Normalize=true
       Safe Fallback Threshold=0
       Safe Fallback Vector=<0 0 0.001>
      )
      Float=12.5
     )
     2=&Multiply_VectorByFloat(
      Vector=&NormalizeVector(
       Vector To Normalize=&Add_Vector(
        A=&Add_Vector(
         A=&Multiply_VectorByFloat(Vector=&ReturnCameraVector(Camera Vectors=::NewEnumerator3, Coordinate Space=::Local), Float=-1)
         B=&Multiply_VectorByFloat(Vector=&ReturnCameraVector(Camera Vectors=::NewEnumerator2, Coordinate Space=::Local), Float=0)
        )
        B=&Multiply_VectorByFloat(Vector=&ReturnCameraVector(Camera Vectors=::NewEnumerator1, Coordinate Space=::Local), Float=-1)
       )
       Safe Normalize=true
       Safe Fallback Threshold=0
       Safe Fallback Vector=<0 0 0.001>
      )
      Float=12.5
     )
     3=?
     Index=&ReturnExecIndex()
     Num=::NewEnumerator1
    )
    WriteAlignment=true
    WriteFacing=false
   )
  ]
  particleUpdateScript [
   &ParticleState()
   &ScaleSpriteSize(
    Scale Sprite Size Mode=::NewEnumerator0
    Uniform Scale Factor=&FloatFromCurve(FloatCurve=DataInterface() {propertyValues {Curve {keys [
         {arriveTangent=21.811, interpMode=RCIM_Cubic, leaveTangent=21.811, tangentMode=RCTM_User, time=0.125, value=0}
         {arriveTangent=37.7491, interpMode=RCIM_Cubic, leaveTangent=37.7491, tangentMode=RCTM_User, time=0.1694, value=1.5545}
         {arriveTangent=3.5399, interpMode=RCIM_Cubic, leaveTangent=3.5399, time=0.236, value=2.828}
         {arriveTangent=-1.8006, interpMode=RCIM_Cubic, leaveTangent=-1.8006, time=0.2922, value=1.9894}
         {arriveTangent=0.5384, interpMode=RCIM_Cubic, leaveTangent=0.5384, time=0.3783, value=2.5718}
         {arriveTangent=-2.5482, interpMode=RCIM_Cubic, leaveTangent=-2.5482, tangentMode=RCTM_User, time=0.8198, value=2.2735}
         {arriveTangent=1.251, tangentMode=RCTM_User, time=1, value=0}
        ]}}})
   ) {module=ScaleSpriteSize002}
   &ScaleColor(
    Scale Alpha=&FloatFromCurve(FloatCurve=DataInterface() {propertyValues {Curve {
        defaultValue=0
        keys [
         {arriveTangent=9.0725, interpMode=RCIM_Cubic, leaveTangent=9.0725, tangentMode=RCTM_User, time=0, value=0}
         {arriveTangent=-0.2441, leaveTangent=-0.2441, time=0.1408, value=0.9847}
         {arriveTangent=-1.146, leaveTangent=-1.146, time=0.95, value=1}
         {arriveTangent=-0.0405, interpMode=RCIM_Cubic, leaveTangent=-0.0405, tangentMode=RCTM_User, time=1, value=0}
        ]
       }}})
    ScaleA=true
    ScaleRGB=true
   )
   &ScaleVelocity(Velocity Scale=&VectorFromFloat(Value=&FloatFromCurve(
      FloatCurve=DataInterface() {propertyValues {Curve {keys [
          <at 0 -0.075>
          <at 0.67 0>
          {interpMode=RCIM_Cubic, leaveTangent=-11.4528, tangentMode=RCTM_Break, time=0.68, value=2}
          {arriveTangent=0.7643, interpMode=RCIM_Cubic, leaveTangent=0.7643, tangentMode=RCTM_User, time=0.9879, value=0}
         ]}}}
      Scale Curve=150
     )))
   &GravityForce(
    Coordinate Space=::Local
    Gravity=&Multiply_VectorByFloat(
     Vector=&NormalizeVector(
      Vector To Normalize=&ReturnCameraVector(Camera Vectors=::NewEnumerator2, Coordinate Space=::Local)
      Safe Normalize=true
      Safe Fallback Threshold=0
      Safe Fallback Vector=<0 0 0.001>
     )
     Float=&FloatFromCurve(FloatCurve=DataInterface() {propertyValues {Curve {keys [<at 0 0>,<at 0.67 0>,<at 0.68 1>,<at 1 2>]}}}, Scale Curve=-2500)
    )
   )
   &SolveForcesAndVelocity(Clamp Velocity=false, Limit Acceleration=false)
  ]
  rendererClasses [&NiagaraSpriteRendererProperties]
  renderers [{rendererClass=&NiagaraSpriteRendererProperties}]
 }
}
summary {
 systemName=NS_ReactFX_ScrollVessel
 emitters [
  {emitterName=Intro, rendererClasses [&NiagaraSpriteRendererProperties]}
  {emitterName=Stars, simTarget=GPUComputeSim, rendererClasses [&NiagaraSpriteRendererProperties]}
  {emitterName=LWE_AirGlow, rendererClasses []}
  {emitterName=LWE_Rays, rendererClasses []}
  {emitterName=LWE_Corona, rendererClasses []}
 ]
 userVariables [
  User.bIsFrontEnd {type=NiagaraBool}
  User.bIsFrontEndPreview {type=NiagaraBool}
 ]
}
