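# Niagara system for a cosmetic companion reaction: a heart inside a blooming flower, sliced apart, with a petting sound.
# Heart, FlowerIntro and PetalBurst build the bloom using staggered sprite animation.
# User.SlashDelay times their lifetimes and the following camera-oriented mesh slash, falling sliced halves and drifting petals, some swaying on waveforms.
# Shared system values supply position, time-of-day tint and, alongside User.GlobalScale, front-end scaling; three glow/flare emitters have no renderers.
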NiagaraSystem /CosmeticCompanions/FX/React/NS_ReactFX_SesameWarp
dependencies {
 usedDataInterfaces [
  &NiagaraDataInterfaceAudioPlayer
  &NiagaraDataInterfaceCurve
  &NiagaraDataInterfaceColorCurve
  &NiagaraDataInterfaceSpriteRendererInfo
  &NiagaraDataInterfaceVector2DCurve
  &NiagaraDataInterfaceCamera
  &NiagaraDataInterfaceVectorCurve
 ]
 usedDynamicInputs [
  &FN_FrontEndFloatModifier
  &Vector4FromCurve
  &FN_FrontEndVector3Modifier
 ]
 usedModules [
  &self:SystemSpawnScript.SystemScriptSource.SystemScriptGraph.NiagaraNodeAssignment_0.SetVariables_#1
  &self:SystemSpawnScript.SystemScriptSource.SystemScriptGraph.NiagaraNodeAssignment_6.SetVariables_#2
  &self:SystemSpawnScript.SystemScriptSource.SystemScriptGraph.NiagaraNodeAssignment_7.SetVariables_#3
  &self:SystemSpawnScript.SystemScriptSource.SystemScriptGraph.NiagaraNodeAssignment_9.SetVariables_#4
  &self:SystemSpawnScript.SystemScriptSource.SystemScriptGraph.NiagaraNodeAssignment_5.SetVariables_#5
  &SystemState
  &self:SystemSpawnScript.SystemScriptSource.SystemScriptGraph.NiagaraNodeAssignment_8.SetVariables_#6
 ]
}
emitters {
 FlowerIntro {
  emitterUpdateScript [
   &EmitterState(Life Cycle Mode=::NewEnumerator1, Loop Behavior=::NewEnumerator1, Loop Duration=1, UseLoopDelay=false)
   &SpawnBurst_Instantaneous(Spawn Count=5, Spawn Group=1, Use Spawn Probability=false, UseLoopCountLimit=false)
  ]
  particleSpawnScript [
   &PartitionParticles(Number Of Partitions=5)
   &InitializeParticle(
    Color=$System.TOD
    Lifetime=&Add_Float(A=$User.SlashDelay, B=0.13)
    Sprite Rotation Angle=$Particles.PartitionParticles.NormalizedPartition
    Sprite Rotation Mode=::NewEnumerator1
    Sprite Size=&Multiply_Vector2D_ByFloat(
     Float=&Multiply_Float(A=$User.GlobalScale, B=$System.FE_Scale)
     Vector2D=&RandomRangeVector2D(
      Minimum=<22 28>
      Maximum=<22 33>
      Randomness Mode=::NewEnumerator1
      Random Seed=&Multiply_Int(A=50, B=&MakeIntFromFloat(Float=$Particles.PartitionParticles.NormalizedPartition))
      Fixed Random Seed=true
     )
    )
    Sprite Size Mode=::NewEnumerator4
    UsePositionOffset=false
    Write Material Random=true
   ) {module=InitializeParticle001}
   /SetVariables(Particles.SortingOrder:NiagaraFloat=&OneMinusFloat(Float=$Particles.PartitionParticles.NormalizedPartition)) {isSetParameters=true}
  ]
  particleUpdateScript [
   &ParticleState()
   &Drag(Drag=&RandomRangeFloat(Minimum=4, Maximum=8), Use Linear Drag=true, Use Rotational Drag=true)
   &SubUVAnimation(
    End Frame Range Override=3
    "Sprite Renderer "=DataInterface() {propertyValues {SpriteRenderer=&self:FlowerIntro_0.NiagaraSpriteRendererProperties_0}}
    SubUV Animation Mode:ENiagara_SubUVLookupModeV2=::NewEnumerator1
    SubUV Curve=DataInterface() {propertyValues {Curve {keys [
        {arriveTangent=1.5752, interpMode=RCIM_Cubic, leaveTangent=1.5752, tangentMode=RCTM_User, time=0.1, value=0}
        {interpMode=RCIM_Cubic, leaveTangent=4.3915, time=0.4, value=1}
       ]}}}
    SubUV Curve Index:NiagaraFloat=&ClampFloat(Float=&Subtract_Float(A=$Particles.NormalizedAge, B=&Multiply_Float(A=$Particles.PartitionParticles.NormalizedPartition, B=0.2)))
    UseEndFrame=true
    UseStartFrame=false
   )
   &SolveForcesAndVelocity(Clamp Velocity=false, Limit Acceleration=false)
   &ScaleColor(
    Scale Alpha=&FloatFromCurve(FloatCurve=DataInterface() {propertyValues {Curve {keys [{arriveTangent=0.317, interpMode=RCIM_Cubic, leaveTangent=0.317, tangentMode=RCTM_User, time=0.5, value=1}, <at 1 1 RCIM_Cubic>]}}})
    Scale RGB=&MakeVectorFromLinearColorRGB(Color=&ColorFromCurve(ColorCurve=DataInterface() {propertyValues {
        AlphaCurve {keys [<at 0 1 RCIM_Cubic>]}
        BlueCurve {keys [<at 0.29 0.7488>,<at 1 0.9328 RCIM_Cubic>]}
        GreenCurve {keys [<at 0.29 0.5391>,<at 1 0.7744 RCIM_Cubic>]}
        RedCurve {keys [<at 0.29 1.5>,<at 1 1.5 RCIM_Cubic>]}
       }}))
    ScaleA=true
    ScaleRGB=true
   )
   &ScaleSpriteSize(Scale Factor=&Vector2DFromFloat(Value=&FloatFromCurve(FloatCurve=DataInterface() {propertyValues {Curve {keys [
          {arriveTangent=0.7974, interpMode=RCIM_Cubic, leaveTangent=0.7974, tangentMode=RCTM_User, time=0, value=0.6}
          {interpMode=RCIM_Cubic, leaveTangent=0.0899, time=0.5, value=1}
         ]}}})))
   /SetVariables(Particles.PivotOffset:Vector2f=/Niagara/DynamicInputs/Add/Add_Vector2(
     A:Vector2f=&Vector2DFromCurve(
      Vector2Curve=DataInterface() {propertyValues {
        XCurve {keys [<at 0 0.5>,<at 1 0.5>,<at 3 0.5>]}
        YCurve {keys [<at 0 0.5 RCIM_Constant>,<at 1 0.58 RCIM_Constant>,<at 2 0.62 RCIM_Constant>]}
       }}
      CurveIndex=$Particles.SubImageIndex
      ScaleCurve=<1 1>
     )
     B:Vector2f=<0 -0.25>
    )) {isSetParameters=true}
   &ScaleSpriteSize(
    Scale Sprite Size Mode=::NewEnumerator0
    Uniform Scale Factor=&FloatFromCurve(FloatCurve=DataInterface() {propertyValues {Curve {keys [<at 0 0 RCIM_Cubic>, {interpMode=RCIM_Cubic, leaveTangent=12.7085, time=0.12, value=1}]}}})
   ) {module=ScaleSpriteSize001}
   /SetVariables(Particles.Position=$System.OffsetPosition) {isSetParameters=true, module=SetVariables_#7}
   /SetVariables(Particles.BacksideColor:LinearColor=&MultiplyLinearColor(
     Linear Color=$Particles.Initial.Color
     Scale Factor=&MakeVector4FromVectorAndFloat(
      "Vector (RGB)"=&MakeVectorFromLinearColorRGB(Color=<rgba 1.2 0.3688 0.5566 1>)
      "Float (Alpha)"=&FloatFromCurve(FloatCurve=DataInterface() {propertyValues {Curve {keys [{arriveTangent=0.317, interpMode=RCIM_Cubic, time=0.5, value=1}, <at 1 0 RCIM_Cubic>]}}})
     )
    )) {isSetParameters=true}
   /SetVariables(Particles.GlowColor=&MakeLinearColorFromVectorAndFloat(
     "Vector (RGB)"=&MakeVectorFromLinearColorRGB(Color=&MultiplyLinearColor(Linear Color=<rgba 5 1.4583 2.2587 1>, Scale Factor=&MakeVector4_FromLinearColor(LinearColor=$System.TOD)))
     "Float (Alpha)"=&FloatFromCurve(FloatCurve=DataInterface() {propertyValues {Curve {keys [<at 0.2 0 RCIM_Cubic>, {arriveTangent=1.25, interpMode=RCIM_Cubic, leaveTangent=1.25, time=0.5, value=1}, <at 1 1 RCIM_Cubic>]}}})
    )) {isSetParameters=true}
   &ScaleColor(
    Scale RGB=&MakeVectorFromLinearColorRGB(Color=/Game/Effects/Niagara/DynamicInputs/TimeOfDay/DaySequenceTimeOfDayColor(Linear Color:LinearColor=<rgba 1 1 1 1>, Time Of Day Influence:NiagaraFloat=1))
    ScaleA=true
    ScaleRGB=true
   ) {module=ScaleColor001}
  ]
  rendererClasses [&NiagaraSpriteRendererProperties]
  renderers [
   {rendererClass=&NiagaraSpriteRendererProperties}
   {rendererClass=&NiagaraSpriteRendererProperties, rendererIndex=1}
   {rendererClass=&NiagaraSpriteRendererProperties, rendererIndex=2}
  ]
 }
 Heart {
  emitterSpawnScript [
   &InitializeAudioPlayer(Audio Player=DataInterface() {propertyValues {
      Attenuation=null
      Concurrency=null
      MaxPlaysPerTick=10
      SoundToPlay=/CosmeticCompanions/Assets/Audio_Shared/Interaction/Petting/SC_Mimosa_Interaction_React_Default
      bAllowLoopingOneShotSounds=false
      bLimitPlaysPerTick=true
      bOnlyActiveDuringGameplay=false
      bStopWhenComponentIsDestroyed=true
      ConfigurationUserParameter {parameter {name=None, typeDefHandle {registeredTypeIndex=1}, varData []}}
      ParameterNames []
     }})
   /Niagara/Modules/Audio/PlayAudio() {module=PlayAudio001}
  ]
  emitterUpdateScript [
   &EmitterState(Life Cycle Mode=::NewEnumerator1, Loop Behavior=::NewEnumerator1, Loop Duration=1, UseLoopDelay=false)
   &SpawnBurst_Instantaneous(Use Spawn Probability=false, UseLoopCountLimit=false)
   &SpawnBurst_Instantaneous(Spawn Group=1, Spawn Time=0.25, Use Spawn Probability=false, UseLoopCountLimit=false) {module=SpawnBurst_Instantaneous001}
  ]
  particleSpawnScript [
   /SetVariables(Particles.VisibilityTag=$Emitter.SpawnGroup) {isSetParameters=true}
   &InitializeParticle(
    Color=&MultiplyLinearColor(
     Linear Color=$System.TOD
     Scale Factor=&MakeVector4_FromLinearColor(LinearColor=&MergeHelper_ColorSwitch(0=<rgba 1 1 1 1>, 1=<rgba 1 0.4271 0.6279 1>, 2=?, 3=?, Index=$Particles.VisibilityTag, Num=::NewEnumerator0))
    )
    Lifetime=&MergeHelper_FloatSwitch(0=&Add_Float(A=$User.SlashDelay, B=0.13), 1=0.5, 2=?, 3=?, Index=$Particles.VisibilityTag, Num=::NewEnumerator0)
    Sprite Rotation Mode=::NewEnumerator0
    Sprite Size Mode=::NewEnumerator3
    Uniform Sprite Size=&Multiply_Float(
     A=&Multiply_Float(
      A=&MergeHelper_FloatSwitch(0=17, 1=34, 2=?, 3=?, Index=$Particles.VisibilityTag, Num=::NewEnumerator0)
      B=$User.GlobalScale
     )
     B=$System.FE_Scale
    )
    UsePositionOffset=false
    Write Material Random=true
   )
  ]
  particleUpdateScript [
   &ParticleState()
   &ScaleSpriteSize(
    Scale Sprite Size Mode=::NewEnumerator0
    Uniform Scale Factor=&MergeHelper_FloatSwitch(
     0=&FloatFromCurve(FloatCurve=DataInterface() {propertyValues {Curve {
         defaultValue=0
         keys [
          {arriveTangent=3.8086, interpMode=RCIM_Cubic, time=0.2, value=0.2}
          {arriveTangent=1.3333, interpMode=RCIM_Cubic, leaveTangent=1.3333, time=0.4374, value=0.5}
          <at 0.8 1 RCIM_Cubic>
         ]
        }}})
     1=&FloatFromCurve(FloatCurve=DataInterface() {propertyValues {Curve {keys [{arriveTangent=3.3458, interpMode=RCIM_Cubic, time=0.25, value=0.3}, <at 0.75 1 RCIM_Cubic>]}}})
     2=?
     3=?
     Index=$Particles.VisibilityTag
     Num=::NewEnumerator0
    )
   ) {module=ScaleSpriteSize002}
   &ScaleColor(
    Scale Alpha=&MergeHelper_FloatSwitch(
     0=&FloatFromCurve(FloatCurve=DataInterface() {propertyValues {Curve {
         defaultValue=0
         keys [
          {arriveTangent=9.0725, interpMode=RCIM_Cubic, leaveTangent=9.0725, tangentMode=RCTM_User, time=0, value=0}
          {arriveTangent=-0.2441, leaveTangent=-0.2441, time=0.1, value=0.9847}
          {arriveTangent=-1.146, leaveTangent=-1.146, time=0.98, value=1}
         ]
        }}})
     1=&FloatFromCurve(FloatCurve=DataInterface() {propertyValues {Curve {
         defaultValue=0
         keys [
          {arriveTangent=0.3137, interpMode=RCIM_Cubic, leaveTangent=0.3137, tangentMode=RCTM_User, time=0, value=0}
          {arriveTangent=2, leaveTangent=2, time=0.1, value=1}
          {arriveTangent=-1.1111, interpMode=RCIM_Cubic, leaveTangent=-1.1111, time=0.5, value=1}
          {arriveTangent=-2.0958, interpMode=RCIM_Cubic, leaveTangent=-2.0958, tangentMode=RCTM_User, time=1, value=0}
         ]
        }}})
     2=?
     3=?
     Index=$Particles.VisibilityTag
     Num=::NewEnumerator0
    )
    Scale RGB=&MakeVectorFromLinearColorRGB(Color=&MergeHelper_ColorSwitch(
      0=<rgba 1 1 1 1>
      1=&ColorFromCurve(ColorCurve=DataInterface() {propertyValues {
         AlphaCurve {keys [<at 0 1 RCIM_Cubic>]}
         BlueCurve {keys [<at 0 2 RCIM_Cubic>,<at 0.1652 1>,<at 1 1 RCIM_Cubic>]}
         GreenCurve {keys [<at 0 2 RCIM_Cubic>,<at 0.1652 1>,<at 1 1 RCIM_Cubic>]}
         RedCurve {keys [<at 0 2 RCIM_Cubic>,<at 0.1652 1>,<at 1 1 RCIM_Cubic>]}
        }})
      2=?
      3=?
      Index=$Particles.VisibilityTag
      Num=::NewEnumerator0
     ))
    ScaleA=true
    ScaleRGB=true
   )
   /SetVariables(Particles.Position=$System.OffsetPosition) {isSetParameters=true}
   &DynamicMaterialParameters(
    Index 0 Param 1=&FloatFromCurve(FloatCurve=DataInterface() {propertyValues {Curve {keys [{arriveTangent=0.6296, interpMode=RCIM_Cubic, leaveTangent=0.6296, tangentMode=RCTM_User, time=0.5, value=1}, <at 1 2>]}}})
    Index 0 Param 2=&FloatFromCurve(FloatCurve=DataInterface() {propertyValues {Curve {keys [{arriveTangent=0.6296, interpMode=RCIM_Cubic, leaveTangent=0.6296, tangentMode=RCTM_User, time=0.5, value=1}, <at 0.6 50>, <at 0.7 1>]}}})
    Index 0 Param 3=1
    Module.Index0Float0Write=true
    Module.Index0Float1Write=true
    Module.Index0Float2Write=true
    Module.Index0Float3Write=true
   )
  ]
  rendererClasses [&NiagaraSpriteRendererProperties]
  renderers [
   {rendererClass=&NiagaraSpriteRendererProperties}
   {rendererClass=&NiagaraSpriteRendererProperties, rendererIndex=1}
  ]
 }
 Intro_Flare {rendererClasses [], renderers []}
 PetalBurst {
  emitterUpdateScript [
   &EmitterState(Life Cycle Mode=::NewEnumerator1, Loop Behavior=::NewEnumerator1, Loop Duration=1, UseLoopDelay=false)
   &SpawnBurst_Instantaneous(Spawn Count=5, Spawn Group=1, Spawn Time=0.1, Use Spawn Probability=false, UseLoopCountLimit=false)
  ]
  particleSpawnScript [
   &PartitionParticles(Number Of Partitions=5)
   &InitializeParticle(
    Color=$System.TOD
    Lifetime=&Add_Float(A=$User.SlashDelay, B=0.15)
    Sprite Rotation Angle=&Add_Float(A=$Particles.PartitionParticles.NormalizedPartition, B=0.6)
    Sprite Rotation Mode=::NewEnumerator1
    Sprite Size Mode=::NewEnumerator3
    Uniform Sprite Size=&Multiply_Float(A=&Multiply_Float(A=3, B=$User.GlobalScale), B=$System.FE_Scale)
    UsePositionOffset=false
    Write Material Random=true
   ) {module=InitializeParticle001}
  ]
  particleUpdateScript [
   &ParticleState()
   &ScaleSpriteSize(
    Scale Factor=&Vector2DFromCurve(
     Vector2Curve=DataInterface() {propertyValues {
       XCurve {keys [<at 0.1 0 RCIM_Cubic>, {arriveTangent=14.9733, interpMode=RCIM_Cubic, leaveTangent=14.9733, tangentMode=RCTM_User, time=0.3, value=1}]}
       YCurve {keys [<at 0.1 0 RCIM_Cubic>,<at 0.3 1 RCIM_Cubic>]}
      }}
     CurveIndex=&Subtract_Float(A=$Particles.NormalizedAge, B=&Multiply_Float(A=$Particles.PartitionParticles.NormalizedPartition, B=0.25))
     ScaleCurve=<1 1>
    )
    Scale Sprite Size Mode=::NewEnumerator2
   )
   &ScaleSpriteSize(
    Scale Sprite Size Mode=::NewEnumerator0
    Uniform Scale Factor=&FloatFromCurve(FloatCurve=DataInterface() {propertyValues {Curve {keys [<at 0.8 1 RCIM_Cubic>,<at 1 0 RCIM_Cubic>]}}})
   ) {module=ScaleSpriteSize001}
   &Drag(Drag=&RandomRangeFloat(Minimum=4, Maximum=8), Use Linear Drag=true, Use Rotational Drag=true)
   &SubUVAnimation(
    End Frame Range Override=2
    "Sprite Renderer "=DataInterface() {propertyValues {SpriteRenderer=&self:PetalBurst_0.NiagaraSpriteRendererProperties_0}}
    SubUV Animation Mode:ENiagara_SubUVLookupModeV2=::NewEnumerator1
    SubUV Curve=DataInterface() {propertyValues {Curve {keys [{arriveTangent=0.8886, interpMode=RCIM_Cubic, time=0.1, value=0}, {interpMode=RCIM_Cubic, leaveTangent=4.3115, time=0.7, value=1}]}}}
    SubUV Curve Index:NiagaraFloat=&ClampFloat(Float=&Subtract_Float(A=$Particles.NormalizedAge, B=&Multiply_Float(A=$Particles.PartitionParticles.NormalizedPartition, B=0.2)))
    UseEndFrame=true
    UseStartFrame=false
   )
   &SolveForcesAndVelocity(Clamp Velocity=false, Limit Acceleration=false)
   &ScaleColor(
    Scale Alpha=&FloatFromCurve(FloatCurve=DataInterface() {propertyValues {Curve {keys [{arriveTangent=0.317, interpMode=RCIM_Cubic, time=0.8, value=1}, <at 1 0 RCIM_Cubic>]}}})
    Scale RGB=&MakeVectorFromLinearColorRGB(Color=&ColorFromCurve(ColorCurve=DataInterface() {propertyValues {
        AlphaCurve {keys [<at 0 1 RCIM_Cubic>]}
        BlueCurve {keys [<at 1 0.8444 RCIM_Cubic>]}
        GreenCurve {keys [<at 1 0.6333 RCIM_Cubic>]}
        RedCurve {keys [<at 1 1.6 RCIM_Cubic>]}
       }}))
    ScaleA=true
    ScaleRGB=true
   )
   /SetVariables(Particles.Position=$System.OffsetPosition) {isSetParameters=true}
   /SetVariables(Particles.PivotOffset:Vector2f=&Lerp_Vector2(
     A=<0.5 1>
     B=<0.5 4>
     Alpha=&FloatFromCurve(
      FloatCurve=DataInterface() {propertyValues {Curve {keys [{arriveTangent=1.6477, interpMode=RCIM_Cubic, leaveTangent=1.6477, tangentMode=RCTM_User, time=0, value=0}, <at 0.6 1 RCIM_Cubic>]}}}
      CurveIndex=&ClampFloat(Float=&Subtract_Float(A=$Particles.NormalizedAge, B=&Multiply_Float(A=$Particles.PartitionParticles.NormalizedPartition, B=0.2)))
     )
    )) {isSetParameters=true}
  ]
  rendererClasses [&NiagaraSpriteRendererProperties]
  renderers [{rendererClass=&NiagaraSpriteRendererProperties}]
 }
 PetalBurstMotion {
  emitterUpdateScript [
   &EmitterState(
    Life Cycle Mode=::NewEnumerator1
    Loop Behavior=::NewEnumerator1
    Loop Delay=&Add_Float(A=$User.SlashDelay, B=0)
    Loop Duration=1
    UseLoopDelay=true
   )
   &SpawnBurst_Instantaneous(
    Spawn Count=&RandomRangeInt(Minimum=6, Maximum=8)
    Spawn Group=1
    Use Spawn Probability=false
    UseLoopCountLimit=false
   )
  ]
  particleSpawnScript [
   /SetVariables(
    Particles.CustomRandom:NiagaraFloat=&RandomRangeFloat(Minimum=0, Maximum=1)
    Particles.SpecialLeafMotion:NiagaraBool=&SetBoolByFloatComparison(A=&RandomRangeFloat(Minimum=0, Maximum=1), B=0.6)
   ) {isSetParameters=true}
   &InitializeParticle(
    Color=$System.TOD
    Lifetime=&MakeCustomFloatFromBool(
     Bool=$Particles.SpecialLeafMotion
     True Float=&RandomRangeFloat(Minimum=2.8, Maximum=3.5)
     False Float=&RandomRangeFloat(Minimum=1.2, Maximum=1.8)
    )
    Position=$System.OffsetPosition
    Position Mode=::NewEnumerator1
    Sprite Rotation Mode=::NewEnumerator2
    Sprite Size Mode=::NewEnumerator3
    Uniform Sprite Size=&Multiply_Float(
     A=&Multiply_Float(A=$User.GlobalScale, B=&RandomRangeFloat(Minimum=2, Maximum=4))
     B=$System.FE_Scale
    )
    UsePositionOffset=false
    Write Material Random=true
   ) {module=InitializeParticle001}
   &ShapeLocation(MaskSpawnGroup=false, Override Seed=false, Sphere Radius=2, UseOwnerScale=true)
   &AddVelocity(
    Velocity=&Multiply_VectorByFloat(
     Vector=&Lerp_Vector(
      A=$Particles.ShapeLocation.ShapeNormal
      B=&Multiply_VectorByFloat(
       Vector=&Add_Vector(
        A=&Multiply_VectorByFloat(Vector=&ReturnCameraVector(Camera Vectors=::NewEnumerator3, Coordinate Space=::Simulation), Float=3)
        B=&Multiply_VectorByFloat(Vector=&ReturnCameraVector(Camera Vectors=::NewEnumerator2, Coordinate Space=::Simulation), Float=1)
       )
       Float=&RandomRangeFloat(Minimum=0.2, Maximum=0.4)
      )
      Alpha=&RandomRangeFloat(Minimum=0.2, Maximum=1)
     )
     Float=$User.GlobalScale
    )
    Velocity Speed Scale=&Multiply_Float(A=&RandomRangeFloat(Minimum=100, Maximum=150), B=$System.FE_Scale)
   )
  ]
  particleUpdateScript [
   &ParticleState()
   &SpriteRotationRate(Rotation Rate=&RandomRangeFloat(Minimum=-100, Maximum=100))
   &Drag(Drag=&RandomRangeFloat(Minimum=3, Maximum=5), Use Linear Drag=true, Use Rotational Drag=true)
   &SubUVAnimation(
    End Frame Range Override=2
    "Sprite Renderer "=DataInterface() {propertyValues {SpriteRenderer=&self:PetalBurstMotion_0.NiagaraSpriteRendererProperties_0}}
    SubUV Animation Mode:ENiagara_SubUVLookupModeV2=::NewEnumerator1
    SubUV Curve=DataInterface() {propertyValues {Curve {keys [{arriveTangent=0.8886, interpMode=RCIM_Cubic, time=0.1, value=0}, {interpMode=RCIM_Cubic, leaveTangent=4.3115, time=0.7, value=1}]}}}
    SubUV Curve Index:NiagaraFloat=&ClampFloat(Float=&Subtract_Float(A=$Particles.NormalizedAge, B=&Multiply_Float(A=$Particles.PartitionParticles.NormalizedPartition, B=0.2)))
    UseEndFrame=true
    UseStartFrame=false
   )
   &GravityForce(Gravity=&MakeCustomVectorFromBool(
     Bool=$Particles.Initial.SpecialLeafMotion
     True Vector=<0 0 -50>
     False Vector=&RandomRangeVector(Minimum=<0 0 -80>, Maximum=<0 0 -120>)
    ))
   &SolveForcesAndVelocity(Clamp Velocity=false, Limit Acceleration=false)
   &ScaleColor(
    Scale Alpha=&FloatFromCurve(FloatCurve=DataInterface() {propertyValues {Curve {keys [{arriveTangent=0.317, interpMode=RCIM_Cubic, time=0.75, value=1}, <at 1 0 RCIM_Cubic>]}}})
    Scale RGB=&MakeVectorFromLinearColorRGB(Color=&ColorFromCurve(ColorCurve=DataInterface() {propertyValues {
        AlphaCurve {keys [<at 0 1 RCIM_Cubic>]}
        BlueCurve {keys [<at 0 0.4969 RCIM_Cubic>]}
        GreenCurve {keys [<at 0 0.4125 RCIM_Cubic>]}
        RedCurve {keys [<at 0 0.7991 RCIM_Cubic>]}
       }}))
    ScaleA=true
    ScaleRGB=true
   )
   /SetVariables(Particles.Position=&AddVectorToPosition(
     Position=<0 0 0>
     Vector=&MakeVector(
      X=&Waveform("Waveform [1]"=::NewEnumerator0, "[1] Frequency"=0.5, "[1] Amplitude Min/Max"=<-1 1>, "[1] Scale"=20)
      Y=0
      Z=&Add_Float(
       A=&FloatFromCurve(
        FloatCurve=DataInterface() {propertyValues {Curve {keys [
            {arriveTangent=3.5366, interpMode=RCIM_Cubic, leaveTangent=3.5366, tangentMode=RCTM_User, time=0, value=0}
            <at 0.5 1 RCIM_Cubic>
            {arriveTangent=-4.1195, interpMode=RCIM_Cubic, leaveTangent=-4.1195, tangentMode=RCTM_User, time=1, value=0}
           ]}}}
        CurveIndex=&Waveform(Global Input Position=1, "Waveform [1]"=::NewEnumerator1, "[1] Frequency"=0.5, "[1] Amplitude Min/Max"=<0 1>)
        Scale Curve=15
       )
       B=&FloatFromCurve(FloatCurve=DataInterface() {propertyValues {Curve {keys [<at 0 0>,<at 1 1>]}}}, Scale Curve=-300)
      )
     )
    )) {enabled=false, isSetParameters=true, module=SetVariables_#7}
   &ScaleSpriteSize(
    Scale Sprite Size Mode=::NewEnumerator0
    Uniform Scale Factor=&FloatFromCurve(FloatCurve=DataInterface() {propertyValues {Curve {keys [<at 0.75 1 RCIM_Cubic>,<at 1 0 RCIM_Cubic>]}}})
   )
   /SetVariables(Particles.CustomPos:NiagaraPosition=&AddVectorToPosition(
     Position=$Particles.Position
     Vector=&MakeCustomVectorFromBool(
      Bool=$Particles.SpecialLeafMotion
      True Vector=&Lerp_Vector(
       A=<0 0 0>
       B=&Add_Vector(
        A=&Multiply_VectorByFloat(
         Vector=&ReturnCameraVector(Camera Vectors=::NewEnumerator3, Coordinate Space=::Simulation)
         Float=&Waveform(
          Global Input Position=$Particles.CustomRandom
          "Waveform [1]"=::NewEnumerator0
          "[1] Frequency"=&Lerp_Float(A=0.3, B=0.5, Alpha=$Particles.Initial.CustomRandom)
          "[1] Amplitude Min/Max"=<-1 1>
          "[1] Scale"=10
         )
        )
        B=&MakeVector(
         X=0
         Y=0
         Z=&FloatFromCurve(
          FloatCurve=DataInterface() {propertyValues {Curve {keys [
              {arriveTangent=3.5366, interpMode=RCIM_Cubic, leaveTangent=3.5366, tangentMode=RCTM_User, time=0, value=0}
              <at 0.5 1 RCIM_Cubic>
              {arriveTangent=-4.1195, interpMode=RCIM_Cubic, leaveTangent=-4.1195, tangentMode=RCTM_User, time=1, value=0}
             ]}}}
          CurveIndex=&Waveform(
           Global Input Position=$Particles.CustomRandom
           "Waveform [1]"=::NewEnumerator1
           "[1] Frequency"=&Lerp_Float(A=0.3, B=0.5, Alpha=$Particles.Initial.CustomRandom)
           "[1] Amplitude Min/Max"=<0 1>
          )
          Scale Curve=3
         )
        )
       )
       Alpha=&FloatFromCurve(FloatCurve=DataInterface() {propertyValues {Curve {defaultValue=1, keys [{tangentMode=RCTM_User, time=0, value=0}, {tangentMode=RCTM_User, time=0.3, value=1}]}}})
      )
      False Vector=<1 0 0>
     )
    )) {isSetParameters=true}
  ]
  rendererClasses [&NiagaraSpriteRendererProperties]
  renderers [{rendererClass=&NiagaraSpriteRendererProperties}]
 }
 Slash {
  emitterUpdateScript [
   &EmitterState(
    Life Cycle Mode=::NewEnumerator1
    Loop Behavior=::NewEnumerator1
    Loop Delay=&Subtract_Float(A=$User.SlashDelay, B=0.05)
    Loop Duration=1
    UseLoopDelay=true
   )
   &SpawnBurst_Instantaneous(Use Spawn Probability=false, UseLoopCountLimit=false) {module=SpawnBurst_Instantaneous001}
   &SpawnBurst_Instantaneous(Spawn Count=1, Spawn Group=1, Use Spawn Probability=true)
   /SetVariables(
    Emitter.FlareOffset:Vector4f=&MakeVector4(
     X=&Lerp_Float(
      A=1
      B=-0.7
      Alpha=&FloatFromCurve(
       FloatCurve=DataInterface() {propertyValues {Curve {keys [{arriveTangent=5.2179, interpMode=RCIM_Cubic, leaveTangent=5.2179, tangentMode=RCTM_User, time=0, value=0}, <at 0.2 1 RCIM_Cubic>]}}}
       CurveIndex=&Subtract_Float(A=$Emitter.LoopedAge, B=0)
      )
     )
     Y=0
     Z=0
     W=0
    )
    Emitter.FlareOffset_Dark:Vector4f=&MakeVector4(
     X=&Lerp_Float(
      A=1
      B=-0.3
      Alpha=&FloatFromCurve(
       FloatCurve=DataInterface() {propertyValues {Curve {keys [{arriveTangent=5.2179, interpMode=RCIM_Cubic, leaveTangent=5.2179, tangentMode=RCTM_User, time=0, value=0}, <at 0.2 1 RCIM_Cubic>]}}}
       CurveIndex=&Subtract_Float(A=$Emitter.LoopedAge, B=0)
      )
     )
     Y=0
     Z=0
     W=0
    )
   ) {isSetParameters=true}
  ]
  particleSpawnScript [
   &InitializeParticle(
    Color=&LerpLinearColors(StartColor=<rgba 1 1 1 1>, EndColor=$System.TOD, LerpFactor=0.3)
    Lifetime=0.5
    Mesh Scale=&Multiply_VectorByFloat(Vector=<0.2 0.15 0.2>, Float=&Multiply_Float(A=$User.GlobalScale, B=$System.FE_Scale))
    Mesh Scale Mode=::NewEnumerator4
    Position=$System.OffsetPosition
    Position Mode=::NewEnumerator1
    Position Offset=&Add_Vector(
     A=<0 0 -1>
     B=&Multiply_VectorByFloat(
      Vector=&Add_Vector(
       A=&Multiply_VectorByFloat(Vector=&ReturnCameraVector(Camera Vectors=::NewEnumerator3, Coordinate Space=::Simulation), Float=3)
       B=&Multiply_VectorByFloat(Vector=&ReturnCameraVector(Camera Vectors=::NewEnumerator2, Coordinate Space=::Simulation), Float=1)
      )
      Float=-2
     )
    )
    UsePositionOffset=true
    Write Material Random=true
   ) {module=InitializeParticle001}
   &CameraOffset(Camera Offset Amount=10)
   &AddVelocity(
    Velocity=&Add_Vector(
     A=&Multiply_VectorByFloat(Vector=&ReturnCameraVector(Camera Vectors=::NewEnumerator3, Coordinate Space=::Simulation), Float=3)
     B=&Multiply_VectorByFloat(Vector=&ReturnCameraVector(Camera Vectors=::NewEnumerator2, Coordinate Space=::Simulation), Float=1)
    )
    Velocity Speed Scale=15
   )
   /SetVariables(Particles.VisibilityTag=$Emitter.SpawnGroup) {isSetParameters=true}
   &InitialMeshOrientation(Mesh Orientation Mode=::NewEnumerator7, Rotation=<-0.02 0 0>, Use Rotation Vector=true)
  ]
  particleUpdateScript [
   &ParticleState()
   &ScaleMeshSize(Scale Factor=&VectorFromCurve(VectorCurve=DataInterface() {propertyValues {XCurve {keys [<at 0 1>,<at 1 1>]}, YCurve {keys [<at 0 1>,<at 1 1>]}, ZCurve {keys [<at 0 1>,<at 0.3 1>,<at 1 1>]}}}))
   &Drag(Use Linear Drag=true, Use Rotational Drag=true)
   &SolveForcesAndVelocity(Clamp Velocity=false, Limit Acceleration=false)
   &ScaleColor(
    Scale Alpha=&FloatFromCurve(FloatCurve=DataInterface() {propertyValues {Curve {keys [
         <at 0 0>
         <at 0.05 1>
         {arriveTangent=-3.5642, interpMode=RCIM_Cubic, leaveTangent=-3.5642, tangentMode=RCTM_User, time=0.5, value=1}
         <at 1 0 RCIM_Cubic>
        ]}}})
    Scale RGB=&Multiply_Vector(
     A=&MakeVectorFromLinearColorRGB(Color=&ColorFromCurve(ColorCurve=DataInterface() {propertyValues {
         AlphaCurve {keys [<at 0 1>]}
         BlueCurve {keys [<at 0.1475 6.339>,<at 0.1834 3.7501>,<at 0.7569 0.7272>]}
         GreenCurve {keys [<at 0.1475 5.1318>,<at 0.1834 2.2917>,<at 0.7569 0.3694>]}
         RedCurve {keys [<at 0.1475 10>,<at 0.1834 5>,<at 0.7569 1>]}
        }}))
     B=&MakeVectorFromLinearColorRGB(Color=&ColorFromCurve(
       ColorCurve=DataInterface() {propertyValues {
         AlphaCurve {keys [<at 0 1>,<at 0.125 1>,<at 0.25 1>,<at 0.375 1>,<at 0.5 1>,<at 0.625 1>,<at 0.75 1>,<at 0.875 1>,<at 1 1>]}
         BlueCurve {keys [<at 0 0.9216>,<at 0.041 0.8963>,<at 0.083 0.8308>,<at 0.125 0.7835>,<at 0.166 0.6172>,<at 0.208 0.5395>,<at 0.25 0.6172>,<at 0.291 0.6939>,<at 0.333 0.7379>,<at 0.375 0.7835>,<at 0.416 0.7991>,<at 0.458 0.8148>,<at 0.5 0.8308>,<at 0.541 0.8148>,<at 0.583 0.7991>,<at 0.625 0.7991>,<at 0.666 0.7529>,<at 0.708 0.7084>,<at 0.75 0.5711>,<at 0.791 0.4735>,<at 0.833 0.3231>,<at 0.875 0.4233>,<at 0.916 0.8388>,<at 0.958 0.9047>,<at 1 0.9216>]}
         GreenCurve {keys [<at 0 0.6308>,<at 0.041 0.5711>,<at 0.083 0.5333>,<at 0.125 0.5583>,<at 0.166 0.5333>,<at 0.208 0.5271>,<at 0.25 0.6376>,<at 0.291 0.7529>,<at 0.333 0.7835>,<at 0.375 0.8148>,<at 0.416 0.7991>,<at 0.458 0.8148>,<at 0.5 0.8469>,<at 0.541 0.7991>,<at 0.583 0.7991>,<at 0.625 0.8148>,<at 0.666 0.7991>,<at 0.708 0.7529>,<at 0.75 0.5711>,<at 0.791 0.5029>,<at 0.833 0.5029>,<at 0.875 0.6172>,<at 0.916 0.5333>,<at 0.958 0.5711>,<at 1 0.6308>]}
         RedCurve {keys [<at 0 0.305>,<at 0.041 0.2623>,<at 0.083 0.2664>,<at 0.125 0.2831>,<at 0.166 0.7084>,<at 0.208 0.6867>,<at 0.25 0.8148>,<at 0.291 0.8879>,<at 0.333 0.8963>,<at 0.375 0.9131>,<at 0.416 0.8714>,<at 0.458 0.8714>,<at 0.5 0.9131>,<at 0.541 0.8632>,<at 0.583 0.8714>,<at 0.625 0.9131>,<at 0.666 0.8963>,<at 0.708 0.8714>,<at 0.75 0.7529>,<at 0.791 0.7529>,<at 0.833 0.8469>,<at 0.875 0.8632>,<at 0.916 0.1878>,<at 0.958 0.1746>,<at 1 0.305>]}
        }}
       CurveIndex=Linked() {linkedVariable {name=NPC.FortniteNPC.TimeOfDay_0-1, type=NiagaraFloat}}
      ))
    )
    ScaleA=true
    ScaleRGB=true
   ) {module=ScaleColor001}
   /SetVariables(Particles.DarkGlowColor:LinearColor=&MakeLinearColorFromVectorAndFloat(
     "Vector (RGB)"=&MakeVectorFromLinearColorRGB(Color=<rgba 0.0208 0.0143 0.0164 1>)
     "Float (Alpha)"=&FloatFromCurve(FloatCurve=DataInterface() {propertyValues {Curve {keys [
          <at 0 0>
          <at 0.05 1>
          {arriveTangent=-3.5642, interpMode=RCIM_Cubic, leaveTangent=-3.5642, tangentMode=RCTM_User, time=0.5, value=1}
          <at 1 0 RCIM_Cubic>
         ]}}})
    )) {isSetParameters=true}
  ]
  rendererClasses [&NiagaraMeshRendererProperties]
  renderers [
   {rendererClass=&NiagaraMeshRendererProperties, rendererIndex=0}
   {rendererClass=&NiagaraMeshRendererProperties, rendererIndex=1}
  ]
 }
 SlashFlare {rendererClasses [], renderers []}
 SlashGlow {rendererClasses [], renderers []}
 SlicedParts {
  emitterUpdateScript [
   &EmitterState(
    Life Cycle Mode=::NewEnumerator1
    Loop Behavior=::NewEnumerator1
    Loop Delay=&Add_Float(A=$User.SlashDelay, B=0.07)
    Loop Duration=1
    UseLoopDelay=true
   )
   &SpawnBurst_Instantaneous(Spawn Count=2, Use Spawn Probability=false, UseLoopCountLimit=false)
   &SpawnBurst_Instantaneous(Spawn Count=2, Spawn Group=1, Use Spawn Probability=false, UseLoopCountLimit=false) {module=SpawnBurst_Instantaneous001}
  ]
  particleSpawnScript [
   /SetVariables(Particles.VisibilityTag=$Emitter.SpawnGroup) {isSetParameters=true}
   &PartitionParticles()
   &InitializeParticle(
    Color=$System.TOD
    Lifetime=&RandomRangeFloat(Minimum=1, Maximum=1.3)
    Sprite Rotation Mode=::NewEnumerator1
    Sprite Size Mode=::NewEnumerator3
    Uniform Sprite Size=&Multiply_Float(
     A=&Multiply_Float(
      A=&MergeHelper_FloatSwitch(0=30, 1=17, 2=?, 3=?, Index=$Particles.VisibilityTag, Num=::NewEnumerator0)
      B=$User.GlobalScale
     )
     B=$System.FE_Scale
    )
    UsePositionOffset=false
    Write Material Random=true
   ) {module=InitializeParticle001}
   /SetVariables(Particles.Position=$System.OffsetPosition) {isSetParameters=true, module=SetVariables_#7}
   &AddVelocity(
    Velocity=&MergeHelper_VectorSwitch(
     0=/Niagara/DynamicInputs/Vectors/RotateVector(
      Vector To Rotate:Vector3f=&ReturnCameraVector(Camera Vectors=::NewEnumerator3, Coordinate Space=::Simulation)
      Yaw:NiagaraFloat=0
      Pitch:NiagaraFloat=30
      Roll:NiagaraFloat=0
      Rotation Rate:NiagaraFloat=1
     )
     1=&Add_Vector(
      A=&RotateVector(
       Vector To Rotate:Vector3f=&ReturnCameraVector(Camera Vectors=::NewEnumerator3, Coordinate Space=::Simulation)
       Yaw:NiagaraFloat=0
       Pitch:NiagaraFloat=&RandomRangeFloat(Minimum=30, Maximum=50)
       Roll:NiagaraFloat=0
       Rotation Rate:NiagaraFloat=1
      )
      B=&RandomRangeVector(Minimum=<-1 -1 0>, Maximum=<1 1 1>)
     )
     2=?
     3=?
     Index=$Particles.VisibilityTag
     Num=::NewEnumerator0
    )
    Velocity Speed Scale=&Multiply_Float(
     A=&MergeHelper_FloatSwitch(0=-1, 1=1, 2=?, 3=?, Index=$Particles.PartitionParticles.Partition, Num=::NewEnumerator0)
     B=&Multiply_Float(A=80, B=$System.FE_Scale)
    )
   )
  ]
  particleUpdateScript [
   &ParticleState()
   &ScaleSpriteSize(Uniform Curve Sprite Scale=DataInterface() {propertyValues {Curve {keys [<at 0.5 1 RCIM_Cubic>,<at 1 0 RCIM_Cubic>]}}})
   &SpriteRotationRate(Rotation Rate=&Multiply_Float(
     A=&MergeHelper_FloatSwitch(0=-1, 1=1, 2=?, 3=?, Index=$Particles.PartitionParticles.Partition, Num=::NewEnumerator0)
     B=50
    ))
   &Drag(
    Drag=&MergeHelper_FloatSwitch(
     0=&RandomRangeFloat(Minimum=4, Maximum=6)
     1=&RandomRangeFloat(Minimum=3, Maximum=5)
     2=?
     3=?
     Index=$StackContext.VisibilityTag
     Num=::NewEnumerator0
    )
    Use Linear Drag=true
    Use Rotational Drag=true
   )
   &GravityForce(Gravity=&MergeHelper_VectorSwitch(
     0=&RandomRangeVector(Minimum=<0 0 -300>, Maximum=<0 0 -500>)
     1=&RandomRangeVector(Minimum=<0 0 -200>, Maximum=<0 0 -300>)
     2=?
     3=?
     Index=$Particles.VisibilityTag
     Num=::NewEnumerator0
    ))
   &SolveForcesAndVelocity(Clamp Velocity=false, Limit Acceleration=false)
   &ScaleColor(
    Scale Alpha=&FloatFromCurve(FloatCurve=DataInterface() {propertyValues {Curve {keys [<at 0.5 1 RCIM_Cubic>,<at 1 0 RCIM_Cubic>]}}})
    Scale RGB=&MakeVectorFromLinearColorRGB(Color=&ColorFromCurve(ColorCurve=DataInterface() {propertyValues {
        AlphaCurve {keys [<at 0 1 RCIM_Cubic>]}
        BlueCurve {keys [<at 0.3091 0.9264>,<at 0.9379 0.7488>]}
        GreenCurve {keys [<at 0.3091 0.7104>,<at 0.9379 0.5391>]}
        RedCurve {keys [<at 0.3091 1.7>,<at 0.9379 1.5>]}
       }}))
    ScaleA=true
    ScaleRGB=true
   )
   &SubUVAnimation(
    Direct Frame Index:NiagaraFloat=&MakeFloatFromInt(Int=$Particles.PartitionParticles.Partition)
    "Sprite Renderer "=DataInterface() {propertyValues {SpriteRenderer=&self:SlicedParts_0.NiagaraSpriteRendererProperties_0}}
    SubUV Animation Mode:ENiagara_SubUVLookupModeV2=::NewEnumerator5
   )
  ]
  rendererClasses [&NiagaraSpriteRendererProperties]
  renderers [
   {rendererClass=&NiagaraSpriteRendererProperties}
   {rendererClass=&NiagaraSpriteRendererProperties, rendererIndex=1}
  ]
 }
}
summary {
 systemName=NS_ReactFX_SesameWarp
 emitters [
  {emitterName=Heart, rendererClasses [&NiagaraSpriteRendererProperties]}
  {emitterName=FlowerIntro, rendererClasses [&NiagaraSpriteRendererProperties]}
  {emitterName=PetalBurst, rendererClasses [&NiagaraSpriteRendererProperties]}
  {emitterName=PetalBurstMotion, rendererClasses [&NiagaraSpriteRendererProperties]}
  {emitterName=Slash, rendererClasses [&NiagaraMeshRendererProperties]}
  {emitterName=SlicedParts, rendererClasses [&NiagaraSpriteRendererProperties]}
  {emitterName=SlashGlow, rendererClasses []}
  {emitterName=SlashFlare, rendererClasses []}
  {emitterName=Intro_Flare, rendererClasses []}
 ]
 userVariables [
  User.BG_Color {type=LinearColor, description="", defaultValue {struct=LinearColor, value=<rgba 0.1406 0.041 0.0495 1>}}
  User.bIsFrontEnd {type=NiagaraBool}
  User.bIsFrontEndPreview {type=NiagaraBool}
  User.Color {type=LinearColor, description="", defaultValue {struct=LinearColor, value=<rgba 0.4219 0.1077 0.1345 1>}}
  User.FrontEndScale {type=NiagaraFloat, defaultValue=0.6, description=""}
  User.GlobalScale {type=NiagaraFloat, defaultValue=1.1, description=""}
  User.SlashDelay {type=NiagaraFloat, defaultValue=0.9, description=""}
 ]
}
