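# Niagara reaction effect for SizeTribe cosmetic companions: a gold-orange mesh travels to a burst point, then pops into a heart shape.
# Intro plays a petting-reaction sound and animates the travel; Pop_FaceCamera bursts as it ends, adding sparkles and glows.
# Partitioned particles are attracted to hard-coded heart points, with a camera-facing position computed from camera vectors.
# Time-of-day and front-end modifiers adjust color and motion; Intro001 and Pop_FaceCamera001 are disabled near-duplicates.
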NiagaraSystem /CosmeticCompanions/FX/React/NS_ReactFX_SizeTribe
dependencies {
 usedDataInterfaces [
  &NiagaraDataInterfaceAudioPlayer
  &NiagaraDataInterfaceCurve
  &NiagaraDataInterfaceVectorCurve
  &NiagaraDataInterfaceVector2DCurve
  &NiagaraDataInterfaceCamera
  &NiagaraDataInterfaceColorCurve
 ]
 usedDynamicInputs [&ColorFromCurve]
 usedModules [
  &self:SystemSpawnScript.SystemScriptSource.SystemScriptGraph.NiagaraNodeAssignment_0.SetVariables_#1
  &self:SystemSpawnScript.SystemScriptSource.SystemScriptGraph.NiagaraNodeAssignment_1.SetVariables_#2
  &self:SystemSpawnScript.SystemScriptSource.SystemScriptGraph.NiagaraNodeAssignment_4.SetVariables_#3
  &self:SystemSpawnScript.SystemScriptSource.SystemScriptGraph.NiagaraNodeAssignment_7.SetVariables_#4
  &SystemState
 ]
}
emitters {
 Intro {
  emitterSpawnScript [
   &InitializeAudioPlayer(Audio Player=DataInterface() {propertyValues {
      Attenuation=null
      Concurrency=null
      MaxPlaysPerTick=10
      SoundToPlay=/CosmeticCompanions/Assets/Audio_Shared/Interaction/Petting/SC_Mimosa_Interaction_React_Default
      bAllowLoopingOneShotSounds=false
      bLimitPlaysPerTick=true
      bOnlyActiveDuringGameplay=false
      bStopWhenComponentIsDestroyed=true
      ConfigurationUserParameter {parameter {name=None, typeDefHandle {registeredTypeIndex=1}, varData []}}
      ParameterNames []
     }})
   /Niagara/Modules/Audio/PlayAudio() {module=PlayAudio001}
  ]
  emitterUpdateScript [
   &EmitterState()
   &SpawnBurst_Instantaneous(Use Spawn Probability=false, UseLoopCountLimit=false)
   /SetVariables(Emitter.Emissive_Intensity:NiagaraFloat=&FNTimeOfDayScalar(Night Scale=0.4, Morning Scale=0.225, Day Scale=0.15, Evening Scale=0.25)) {isSetParameters=true}
  ]
  particleSpawnScript [&InitializeParticle(
    Color=&FNTimeOfDayColor(
     Night Color=<rgba 2.3125 1.1562 0 1>
     Morning Color=<rgba 2.3125 1.2719 0 1>
     Day Color=<rgba 2.3125 1.1562 0 1>
     Evening Color=<rgba 2.3125 0.9635 0 1>
    )
    Lifetime=$System.IntroLength
    Mesh Scale Mode=::NewEnumerator3
    Mesh Uniform Scale=0.7
    Sprite Rotation Mode=::NewEnumerator0
    Sprite Size Mode=::NewEnumerator3
    Uniform Sprite Size=15
    UsePositionOffset=false
    Write Material Random=true
   )]
  particleUpdateScript [
   &ParticleState()
   &ScaleMeshSize(Scale Factor=&VectorFromFloat(Value=&FloatFromCurve(FloatCurve=DataInterface() {propertyValues {Curve {
         defaultValue=1
         keys [
          {arriveTangent=2.5713, interpMode=RCIM_Cubic, leaveTangent=2.5713, tangentMode=RCTM_User, time=0, value=0}
          {arriveTangent=1.3436, interpMode=RCIM_Cubic, leaveTangent=1.3436, tangentMode=RCTM_User, time=0.4551, value=1}
         ]
        }}})))
   &UpdateMeshOrientation(Rotation Rate=-2, Rotation Vector=<0 0 1>)
   &ScaleSpriteSize(
    Scale Sprite Size Mode=::NewEnumerator0
    Uniform Scale Factor=&Add_Float(
     A=0.5
     B=&FloatFromCurve(FloatCurve=DataInterface() {propertyValues {Curve {
         defaultValue=0
         keys [
          {arriveTangent=0.8193, interpMode=RCIM_Cubic, leaveTangent=0.8193, tangentMode=RCTM_User, time=0, value=0}
          {arriveTangent=2.055, interpMode=RCIM_Cubic, leaveTangent=2.055, tangentMode=RCTM_User, time=0.7247, value=0.9545}
          {arriveTangent=-0.9053, interpMode=RCIM_Cubic, leaveTangent=-0.9053, tangentMode=RCTM_User, time=0.8783, value=2}
          {arriveTangent=-0.1202, interpMode=RCIM_Cubic, leaveTangent=-0.1202, tangentMode=RCTM_User, time=1, value=1}
         ]
        }}})
    )
   ) {module=ScaleSpriteSize002}
   &ScaleColor(
    Scale RGB=&FN_FrontEndVector3Modifier(
     Gameplay=&FN_Select_Vector_from_Time_Of_Day(Morning=<1.115 1.15 1.15>, Day=<1 1 1>, Evening=<1 1 1>, Night=<2.65 2.65 2.65>)
     Front End=<3.2 2.75 2.75>
    )
    ScaleA=true
    ScaleRGB=true
   )
   &OffsetPosition(Position Offset=&Lerp_Vector(
     A=$System.HeartIntroSpawnPosition
     B=&Add_Vector(A=$System.BurstPosition, B=<0 0 -5>)
     Alpha=&FloatFromCurve(FloatCurve=DataInterface() {propertyValues {Curve {keys [
          {arriveTangent=0.7913, interpMode=RCIM_Cubic, leaveTangent=0.7913, tangentMode=RCTM_User, time=0, value=0}
          {arriveTangent=-0.1074, tangentMode=RCTM_User, time=1, value=1}
         ]}}})
    ))
   /SetVariables(
    Particles.Glow Color=&Multiply_LinearColorByFloat(Linear Color=$System.Layer 2 Color, Float=0.3)
    Particles.Glow Size=&Multiply_Vector2D(A=$Particles.SpriteSize, B=<0.6 0.6>)
   ) {isSetParameters=true, module=SetVariables_#5}
  ]
  rendererClasses [&NiagaraMeshRendererProperties]
  renderers [
   {rendererClass=&NiagaraMeshRendererProperties, rendererIndex=0}
   {rendererClass=&NiagaraMeshRendererProperties, rendererIndex=1}
  ]
 }
 Intro001 {
  bEnabled=false
  emitterSpawnScript [
   &InitializeAudioPlayer(Audio Player=DataInterface() {propertyValues {
      Attenuation=null
      Concurrency=null
      MaxPlaysPerTick=10
      SoundToPlay=&SC_Mimosa_Interaction_React_Default
      bAllowLoopingOneShotSounds=false
      bLimitPlaysPerTick=true
      bOnlyActiveDuringGameplay=false
      bStopWhenComponentIsDestroyed=true
      ConfigurationUserParameter {parameter {name=None, typeDefHandle {registeredTypeIndex=1}, varData []}}
      ParameterNames []
     }})
   &PlayAudio() {module=PlayAudio001}
  ]
  emitterUpdateScript [
   &EmitterState()
   &SpawnBurst_Instantaneous(Use Spawn Probability=false, UseLoopCountLimit=false)
  ]
  particleSpawnScript [&InitializeParticle(
    Color=&FNTimeOfDayColor(
     Night Color=<rgba 2.3125 1.1562 0 1>
     Morning Color=<rgba 2.3125 1.2719 0 1>
     Day Color=<rgba 2.3125 1.1562 0 1>
     Evening Color=<rgba 2.3125 0.9635 0 1>
    )
    Lifetime=$System.IntroLength
    Mesh Scale Mode=::NewEnumerator3
    Mesh Uniform Scale=0.7
    Sprite Rotation Mode=::NewEnumerator0
    Sprite Size Mode=::NewEnumerator3
    Uniform Sprite Size=15
    UsePositionOffset=false
    Write Material Random=true
   )]
  particleUpdateScript [
   &ParticleState()
   &ScaleMeshSize(Scale Factor=&VectorFromFloat(Value=&FloatFromCurve(FloatCurve=DataInterface() {propertyValues {Curve {
         defaultValue=1
         keys [
          {arriveTangent=2.5713, interpMode=RCIM_Cubic, leaveTangent=2.5713, tangentMode=RCTM_User, time=0, value=0}
          {arriveTangent=1.3436, interpMode=RCIM_Cubic, leaveTangent=1.3436, tangentMode=RCTM_User, time=0.4551, value=1}
         ]
        }}})))
   &UpdateMeshOrientation(Rotation Rate=-2, Rotation Vector=<0 0 1>)
   &ScaleSpriteSize(
    Scale Sprite Size Mode=::NewEnumerator0
    Uniform Scale Factor=&Add_Float(
     A=0.5
     B=&FloatFromCurve(FloatCurve=DataInterface() {propertyValues {Curve {
         defaultValue=0
         keys [
          {arriveTangent=0.8193, interpMode=RCIM_Cubic, leaveTangent=0.8193, tangentMode=RCTM_User, time=0, value=0}
          {arriveTangent=2.055, interpMode=RCIM_Cubic, leaveTangent=2.055, tangentMode=RCTM_User, time=0.7247, value=0.9545}
          {arriveTangent=-0.9053, interpMode=RCIM_Cubic, leaveTangent=-0.9053, tangentMode=RCTM_User, time=0.8783, value=2}
          {arriveTangent=-0.1202, interpMode=RCIM_Cubic, leaveTangent=-0.1202, tangentMode=RCTM_User, time=1, value=1}
         ]
        }}})
    )
   ) {module=ScaleSpriteSize002}
   &ScaleColor(
    Scale RGB=&FN_FrontEndVector3Modifier(
     Gameplay=&FN_Select_Vector_from_Time_Of_Day(Morning=<1.115 1.15 1.15>, Day=<1 1 1>, Evening=<1 1 1>, Night=<2.65 2.65 2.65>)
     Front End=<3.2 2.75 2.75>
    )
    ScaleA=true
    ScaleRGB=true
   )
   &OffsetPosition(Position Offset=&Lerp_Vector(
     A=$System.HeartIntroSpawnPosition
     B=&Add_Vector(A=$System.BurstPosition, B=<0 0 -5>)
     Alpha=&FloatFromCurve(FloatCurve=DataInterface() {propertyValues {Curve {keys [
          {arriveTangent=0.7913, interpMode=RCIM_Cubic, leaveTangent=0.7913, tangentMode=RCTM_User, time=0, value=0}
          {arriveTangent=-0.1074, tangentMode=RCTM_User, time=1, value=1}
         ]}}})
    ))
   /SetVariables(
    Particles.Glow Color=&Multiply_LinearColorByFloat(Linear Color=$System.Layer 2 Color, Float=0.3)
    Particles.Glow Size=&Multiply_Vector2D(A=$Particles.SpriteSize, B=<0.6 0.6>)
   ) {isSetParameters=true, module=SetVariables_#5}
  ]
  rendererClasses [&NiagaraMeshRendererProperties]
  renderers [
   {rendererClass=&NiagaraMeshRendererProperties, rendererIndex=0}
   {rendererClass=&NiagaraMeshRendererProperties, rendererIndex=1}
  ]
 }
 Pop_FaceCamera {
  emitterUpdateScript [
   &EmitterState()
   &SpawnBurst_Instantaneous(
    Spawn Count=16
    Spawn Time=&Subtract_Float(A=$System.IntroLength, B=0)
    Use Spawn Probability=false
    UseLoopCountLimit=false
   )
   &SpawnBurst_Instantaneous(
    Spawn Count=&RandomRangeInt(Minimum=30, Maximum=50)
    Spawn Group=1
    Spawn Time=&Subtract_Float(A=$System.IntroLength, B=0.03)
    Use Spawn Probability=false
    UseLoopCountLimit=false
   ) {module=SpawnBurst_Instantaneous001}
   &SpawnBurst_Instantaneous(
    Spawn Group=2
    Spawn Time=&Subtract_Float(A=$System.IntroLength, B=0.03)
    Use Spawn Probability=false
    UseLoopCountLimit=false
   ) {module=SpawnBurst_Instantaneous002}
   &SpawnBurst_Instantaneous(
    Spawn Group=3
    Spawn Time=&Subtract_Float(A=$System.IntroLength, B=0.03)
    Use Spawn Probability=false
    UseLoopCountLimit=false
   ) {module=SpawnBurst_Instantaneous003}
   &SpawnBurst_Instantaneous(
    Spawn Group=3
    Spawn Time=&Subtract_Float(A=$System.IntroLength, B=0.03)
    Use Spawn Probability=false
    UseLoopCountLimit=false
   ) {module=SpawnBurst_Instantaneous004}
   /SetVariables(Emitter.Emissive_Intensity:NiagaraFloat=&FNTimeOfDayScalar(Night Scale=0.4, Morning Scale=0.225, Day Scale=0.15, Evening Scale=0.25)) {isSetParameters=true}
  ]
  particleSpawnScript [
   /SetVariables(Particles.VisibilityTag=&SpawnGroup()) {isSetParameters=true, module=SetVariables_#6}
   &PartitionParticles(Number Of Partitions=16, Particles Per Partition=1, Partition Mode:ENiagara_PartitionMode_v2=::NewEnumerator1)
   &InitializeParticle(
    Color=&MergeHelper_ColorSwitch(
     0=&FNTimeOfDayColor(
      Night Color=<rgba 2.3125 1.1562 0 1>
      Morning Color=<rgba 2.3125 1.2719 0 1>
      Day Color=<rgba 2.3125 1.1562 0 1>
      Evening Color=<rgba 2.3125 0.9635 0 1>
     )
     1=&RandomRangeLinearColor(Minimum=<rgba 1 0.5281 0.0563 0.5>, Maximum=<rgba 1 0.45 0 0.5>)
     2=<rgba 3 1.8435 0.2244 1>
     3=<rgba 3 1.8438 0.225 1>
     Index=$Particles.VisibilityTag
     Num=::NewEnumerator2
    )
    Lifetime=&MergeHelper_FloatSwitch(
     0=&RandomRangeFloat(Minimum=0.75, Maximum=1.2, Randomness Mode=::NewEnumerator2, Random Seed=?)
     1=&RandomRangeFloat(Minimum=0.4, Maximum=1.25)
     2=0.1
     3=0.1
     Index=$Particles.VisibilityTag
     Num=::NewEnumerator2
    )
    Mesh Scale Mode=::NewEnumerator3
    Mesh Uniform Scale=0.25
    Position Offset=&MergeHelper_VectorSwitch(
     0=<0 0 0>
     1=<0 0 0>
     2=&AddVectorToPosition(Position=Linked() {linkedVariable {name=System.BurstPosition, type=Vector3f}}, Vector=<0 0 0>)
     3=&AddVectorToPosition(Position=Linked() {linkedVariable {name=System.BurstPosition, type=Vector3f}}, Vector=<0 0 0>)
     Index=$Particles.VisibilityTag
     Num=::NewEnumerator2
    )
    Sprite Rotation Angle=-20
    Sprite Rotation Mode=::NewEnumerator0
    Sprite Size Mode=::NewEnumerator3
    Uniform Sprite Size=&Lerp_Float(
     A=10
     B=50
     Alpha=&FloatFromCurve(
      FloatCurve=DataInterface() {propertyValues {Curve {keys [
          {interpMode=RCIM_Cubic, tangentMode=RCTM_User, time=0, value=0.2}
          {arriveTangent=1.1314, interpMode=RCIM_Cubic, leaveTangent=1.1314, tangentMode=RCTM_User, time=0.9, value=0.6}
          {interpMode=RCIM_Cubic, tangentMode=RCTM_User, time=1, value=1}
         ]}}}
      CurveIndex=&RandomRangeFloat(Minimum=0, Maximum=1)
     )
    )
    UsePositionOffset=true
    Write Material Random=true
   )
   /SetVariables(
    Particles.GlitterSubUV:NiagaraFloat=&RandomRangeFloat(Minimum=0, Maximum=3)
    Particles.InitialRoundGlowSize:Vector2f=<100 100>
    Particles.InitialStarSize:Vector2f=&Multiply_Vector2D_ByFloat(Float=0.6, Vector2D=<30 40>)
   ) {isSetParameters=true, module=SetVariables_#7}
   &ShapeLocation(
    MaskSpawnGroup=false
    Offset=&MergeHelper_VectorSwitch(0=$System.BurstPosition, 1=$System.BurstPosition, 2=$System.BurstPosition, 3=?, Index=$Particles.VisibilityTag, Num=::NewEnumerator1)
    Offset Mode=::NewEnumerator0
    Sphere Distribution=::NewEnumerator3
    Sphere Radius=&MergeHelper_FloatSwitch(0=2, 1=1, 2=?, 3=?, Index=$Particles.VisibilityTag, Num=::NewEnumerator0)
    Uniform Particle Index:NiagaraInt32=&MergeHelper_IntSwitch(
     0=$Particles.PartitionParticles.Partition
     1=&ReturnExecIndex()
     2=0
     3=0
     4=?
     5=?
     Index=$Particles.VisibilityTag
     Num=::NewEnumerator2
    )
    Uniform Spiral Amount=0.5
    UseOwnerScale=true
   )
   &AddVelocity(
    Module.Use Falloff Distance=false
    Origin Offset=&MergeHelper_VectorSwitch(
     0=&AddVectorToPosition(Position=<0 0 -4>, Vector=$System.BurstPosition)
     1=&AddVectorToPosition(Position=<0 0 -0.6>, Vector=$System.BurstPosition)
     2=?
     3=?
     Index=$Particles.VisibilityTag
     Num=::NewEnumerator0
    )
    Override Seed=false
    Randomness Mode:ENiagaraRandomnessMode=::NewEnumerator2
    Velocity Mode=::NewEnumerator1
    Velocity Speed=&MergeHelper_FloatSwitch(
     0=0
     1=&RandomRangeFloat(
      Minimum=30
      Maximum=&FN_FrontEndFloatModifier(Gameplay=300, Front End=200)
      Randomness Mode=::NewEnumerator2
      Random Seed=?
     )
     2=?
     3=?
     Index=$Particles.VisibilityTag
     Num=::NewEnumerator0
    )
   )
   &AddVelocity(Module.Use Falloff Distance=false, Override Seed=false, Velocity Mode=::NewEnumerator1) {enabled=false, module=AddVelocity001}
   &InitialMeshOrientation(Mesh Axis To Orient=<0 1 0>, Mesh Orientation Mode=::NewEnumerator2, Use Rotation Vector=false)
  ]
  particleUpdateScript [
   &ParticleState()
   &ScaleColor(
    Scale Alpha=&MergeHelper_FloatSwitch(
     0=1
     1=&Multiply_Float(
      A=0.25
      B=&FloatFromCurve(
       FloatCurve=DataInterface() {propertyValues {Curve {keys [<at 0 1>,<at 1 0>]}}}
       Scale Curve=&Waveform(
        Global Amplitude Scale=10
        "Waveform [1]"=::NewEnumerator0
        "[1] Phase Scale"=&RandomRangeFloat(Minimum=2, Maximum=4)
        "[1] Frequency"=&RandomRangeFloat(Minimum=0.75, Maximum=1)
        "[1] Amplitude Min/Max"=<0 1>
       )
      )
     )
     2=1
     3=1
     Index=$Particles.VisibilityTag
     Num=::NewEnumerator2
    )
    Scale RGB=&MergeHelper_VectorSwitch(
     0=&FN_FrontEndVector3Modifier(
      Gameplay=&FN_Select_Vector_from_Time_Of_Day(Morning=<1.115 1.15 1.15>, Day=<1 1 1>, Evening=<1 1 1>, Night=<2.65 2.65 2.65>)
      Front End=<3.2 2.75 2.75>
     )
     1=<1 1 1>
     2=<1 1 1>
     3=<1 1 1>
     Index=0
     Num=::NewEnumerator2
    )
    ScaleA=true
    ScaleRGB=true
   )
   &GravityForce(Gravity=&Multiply_VectorByFloat(
     Vector=&MergeHelper_VectorSwitch(
      0=&RandomRangeVector(Minimum=<0 0 -80>, Maximum=<0 0 -110>, Randomness Mode=::NewEnumerator1, Random Seed=10, Fixed Random Seed=true)
      1=&RandomRangeVector(Minimum=<0 0 -250>, Maximum=<0 0 -50>)
      2=?
      3=?
      Index=$Particles.VisibilityTag
      Num=::NewEnumerator0
     )
     Float=&MergeHelper_FloatSwitch(
      0=&FloatFromCurve(
       FloatCurve=DataInterface() {propertyValues {Curve {keys [{arriveTangent=1.7842, interpMode=RCIM_Cubic, leaveTangent=1.7842, tangentMode=RCTM_User, time=0, value=1}, <at 1 2 RCIM_Cubic>]}}}
       Scale Curve=1.6
      )
      1=1
      2=?
      3=?
      Index=$Particles.VisibilityTag
      Num=::NewEnumerator0
     )
    ))
   &Drag(
    Drag=&MergeHelper_FloatSwitch(
     0=&FloatFromCurve(
      FloatCurve=DataInterface() {propertyValues {Curve {keys [<at 0 0 RCIM_Cubic>,<at 0.176 0.7545 RCIM_Cubic>,<at 0.3 0 RCIM_Cubic>]}}}
      Scale Curve=3
     )
     1=&RandomRangeFloat(Minimum=5, Maximum=12)
     2=?
     3=?
     Index=$Particles.VisibilityTag
     Num=::NewEnumerator0
    )
    Use Linear Drag=true
    Use Rotational Drag=true
   )
   &PointAttractionForce(
    Attraction Radius=15
    Attraction Strength=&MergeHelper_FloatSwitch(
     0=&FloatFromCurve(
      FloatCurve=DataInterface() {propertyValues {Curve {keys [
          {arriveTangent=36.0485, interpMode=RCIM_Cubic, leaveTangent=36.0485, tangentMode=RCTM_User, time=0, value=0}
          {arriveTangent=0.4397, interpMode=RCIM_Cubic, leaveTangent=0.4397, tangentMode=RCTM_User, time=0.089, value=1}
          {arriveTangent=-0.6185, interpMode=RCIM_Cubic, leaveTangent=-0.6185, tangentMode=RCTM_User, time=0.3, value=0}
         ]}}}
      Scale Curve=100
     )
     1=0
     2=0
     3=0
     Index=$Particles.VisibilityTag
     Num=::NewEnumerator2
    )
    Attractor Position Offset:Vector3f=&AddVectorToPosition(
     Position=&Multiply_Vector(
      A=/Niagara/DynamicInputs/Helpers/MaskPosition(
       Mask Source:NiagaraInt32=1
       Mask Parameter:NiagaraInt32=$Particles.PartitionParticles.Partition
       Mask:NiagaraInt32=0
       Masked Position:NiagaraPosition=<-48 0 17>
       Passthrough Position:NiagaraPosition=&MaskPosition(
        Mask Source:NiagaraInt32=1
        Mask Parameter:NiagaraInt32=$Particles.PartitionParticles.Partition
        Mask:NiagaraInt32=1
        Masked Position:NiagaraPosition=<-40 0 -29>
        Passthrough Position:NiagaraPosition=&MaskPosition(
         Mask Source:NiagaraInt32=1
         Mask Parameter:NiagaraInt32=$Particles.PartitionParticles.Partition
         Mask:NiagaraInt32=2
         Masked Position:NiagaraPosition=<40 0 -29>
         Passthrough Position:NiagaraPosition=&MaskPosition(
          Mask Source:NiagaraInt32=1
          Mask Parameter:NiagaraInt32=$Particles.PartitionParticles.Partition
          Mask:NiagaraInt32=3
          Masked Position:NiagaraPosition=<-25 0 -47>
          Passthrough Position:NiagaraPosition=&MaskPosition(
           Mask Source:NiagaraInt32=1
           Mask Parameter:NiagaraInt32=$Particles.PartitionParticles.Partition
           Mask:NiagaraInt32=4
           Masked Position:NiagaraPosition=<0 0 -92>
           Passthrough Position:NiagaraPosition=&MaskPosition(
            Mask Source:NiagaraInt32=1
            Mask Parameter:NiagaraInt32=$Particles.PartitionParticles.Partition
            Mask:NiagaraInt32=5
            Masked Position:NiagaraPosition=<14 0 10>
            Passthrough Position:NiagaraPosition=&MaskPosition(
             Mask Source:NiagaraInt32=1
             Mask Parameter:NiagaraInt32=$Particles.PartitionParticles.Partition
             Mask:NiagaraInt32=6
             Masked Position:NiagaraPosition=<0 0 -10>
             Passthrough Position:NiagaraPosition=&MaskPosition(
              Mask Source:NiagaraInt32=1
              Mask Parameter:NiagaraInt32=$Particles.PartitionParticles.Partition
              Mask:NiagaraInt32=7
              Masked Position:NiagaraPosition=<-12 0 10>
              Passthrough Position:NiagaraPosition=&MaskPosition(
               Mask Source:NiagaraInt32=1
               Mask Parameter:NiagaraInt32=$Particles.PartitionParticles.Partition
               Mask:NiagaraInt32=8
               Masked Position:NiagaraPosition=<48 0 17>
               Passthrough Position:NiagaraPosition=&MaskPosition(
                Mask Source:NiagaraInt32=1
                Mask Parameter:NiagaraInt32=$Particles.PartitionParticles.Partition
                Mask:NiagaraInt32=9
                Masked Position:NiagaraPosition=<28 0 28>
                Passthrough Position:NiagaraPosition=&MaskPosition(
                 Mask Source:NiagaraInt32=1
                 Mask Parameter:NiagaraInt32=$Particles.PartitionParticles.Partition
                 Mask:NiagaraInt32=10
                 Masked Position:NiagaraPosition=<-28 0 28>
                 Passthrough Position:NiagaraPosition=&MaskPosition(
                  Mask Source:NiagaraInt32=1
                  Mask Parameter:NiagaraInt32=$Particles.PartitionParticles.Partition
                  Mask:NiagaraInt32=11
                  Masked Position:NiagaraPosition=<50 0 -7>
                  Passthrough Position:NiagaraPosition=&MaskPosition(
                   Mask Source:NiagaraInt32=1
                   Mask Parameter:NiagaraInt32=$Particles.PartitionParticles.Partition
                   Mask:NiagaraInt32=12
                   Masked Position:NiagaraPosition=<12 0 -70>
                   Passthrough Position:NiagaraPosition=&MaskPosition(
                    Mask Source:NiagaraInt32=1
                    Mask Parameter:NiagaraInt32=$Particles.PartitionParticles.Partition
                    Mask:NiagaraInt32=13
                    Masked Position:NiagaraPosition=<-50 0 -7>
                    Passthrough Position:NiagaraPosition=&MaskVector(Mask Source=1, Mask Parameter=$Particles.PartitionParticles.Partition, Mask=14, Masked Vector=<-12 0 -70>, Passthrough Vector=<25 0 -47>)
                   )
                  )
                 )
                )
               )
              )
             )
            )
           )
          )
         )
        )
       )
      )
      B=&Multiply_VectorByFloat(Vector=<0.25 0.15 0.2>, Float=&FN_FrontEndFloatModifier(Gameplay=0.55, Front End=0.4))
     )
     Vector=&AddVectorToPosition(
      Position=Linked() {linkedVariable {name=System.BurstPosition, type=Vector3f}}
      Vector=&VectorFromCurve(
       VectorCurve=DataInterface() {propertyValues {XCurve {keys [<at 0 0>,<at 1 0>]}, YCurve {keys [<at 0 0>,<at 1 0>]}, ZCurve {keys [<at 0 1 RCIM_Cubic>,<at 0.7003 -1 RCIM_Cubic>]}}}
       Scale Curve=<1 1 5>
      )
     )
    )
    Attractor Position Offset Coordinate Space:ENiagaraCoordinateSpace=::Local
    Falloff Exponent=1
    Kill Within Radius=false
    Module.Use Falloff=true
    UseAttractorOffset=true
   )
   /SetVariables(Particles.PositionHeart:NiagaraPosition=&Lerp_Vector(
     A=Linked() {linkedVariable {name=Particles.Position, type=NiagaraPosition}}
     B=&Multiply_Vector(
      A=&AddVectorToPosition(
       Position=&MaskPosition(
        Mask Source:NiagaraInt32=1
        Mask Parameter:NiagaraInt32=$Particles.PartitionParticles.Partition
        Mask:NiagaraInt32=0
        Masked Position:NiagaraPosition=<-48 0 17>
        Passthrough Position:NiagaraPosition=&MaskPosition(
         Mask Source:NiagaraInt32=1
         Mask Parameter:NiagaraInt32=$Particles.PartitionParticles.Partition
         Mask:NiagaraInt32=1
         Masked Position:NiagaraPosition=<-40 0 -29>
         Passthrough Position:NiagaraPosition=&MaskPosition(
          Mask Source:NiagaraInt32=1
          Mask Parameter:NiagaraInt32=$Particles.PartitionParticles.Partition
          Mask:NiagaraInt32=2
          Masked Position:NiagaraPosition=<40 0 -29>
          Passthrough Position:NiagaraPosition=&MaskPosition(
           Mask Source:NiagaraInt32=1
           Mask Parameter:NiagaraInt32=$Particles.PartitionParticles.Partition
           Mask:NiagaraInt32=3
           Masked Position:NiagaraPosition=<-25 0 -47>
           Passthrough Position:NiagaraPosition=&MaskPosition(
            Mask Source:NiagaraInt32=1
            Mask Parameter:NiagaraInt32=$Particles.PartitionParticles.Partition
            Mask:NiagaraInt32=4
            Masked Position:NiagaraPosition=<0 0 -92>
            Passthrough Position:NiagaraPosition=&MaskPosition(
             Mask Source:NiagaraInt32=1
             Mask Parameter:NiagaraInt32=$Particles.PartitionParticles.Partition
             Mask:NiagaraInt32=5
             Masked Position:NiagaraPosition=<14 0 10>
             Passthrough Position:NiagaraPosition=&MaskPosition(
              Mask Source:NiagaraInt32=1
              Mask Parameter:NiagaraInt32=$Particles.PartitionParticles.Partition
              Mask:NiagaraInt32=6
              Masked Position:NiagaraPosition=<0 0 -10>
              Passthrough Position:NiagaraPosition=&MaskPosition(
               Mask Source:NiagaraInt32=1
               Mask Parameter:NiagaraInt32=$Particles.PartitionParticles.Partition
               Mask:NiagaraInt32=7
               Masked Position:NiagaraPosition=<-12 0 10>
               Passthrough Position:NiagaraPosition=&MaskPosition(
                Mask Source:NiagaraInt32=1
                Mask Parameter:NiagaraInt32=$Particles.PartitionParticles.Partition
                Mask:NiagaraInt32=8
                Masked Position:NiagaraPosition=<48 0 17>
                Passthrough Position:NiagaraPosition=&MaskPosition(
                 Mask Source:NiagaraInt32=1
                 Mask Parameter:NiagaraInt32=$Particles.PartitionParticles.Partition
                 Mask:NiagaraInt32=9
                 Masked Position:NiagaraPosition=<28 0 28>
                 Passthrough Position:NiagaraPosition=&MaskPosition(
                  Mask Source:NiagaraInt32=1
                  Mask Parameter:NiagaraInt32=$Particles.PartitionParticles.Partition
                  Mask:NiagaraInt32=10
                  Masked Position:NiagaraPosition=<-28 0 28>
                  Passthrough Position:NiagaraPosition=&MaskPosition(
                   Mask Source:NiagaraInt32=1
                   Mask Parameter:NiagaraInt32=$Particles.PartitionParticles.Partition
                   Mask:NiagaraInt32=11
                   Masked Position:NiagaraPosition=<50 0 -7>
                   Passthrough Position:NiagaraPosition=&MaskPosition(
                    Mask Source:NiagaraInt32=1
                    Mask Parameter:NiagaraInt32=$Particles.PartitionParticles.Partition
                    Mask:NiagaraInt32=12
                    Masked Position:NiagaraPosition=<12 0 -70>
                    Passthrough Position:NiagaraPosition=&MaskPosition(
                     Mask Source:NiagaraInt32=1
                     Mask Parameter:NiagaraInt32=$Particles.PartitionParticles.Partition
                     Mask:NiagaraInt32=13
                     Masked Position:NiagaraPosition=<-50 0 -7>
                     Passthrough Position:NiagaraPosition=&MaskVector(Mask Source=1, Mask Parameter=$Particles.PartitionParticles.Partition, Mask=14, Masked Vector=<-12 0 -70>, Passthrough Vector=<25 0 -47>)
                    )
                   )
                  )
                 )
                )
               )
              )
             )
            )
           )
          )
         )
        )
       )
       Vector=<0 0 20>
      )
      B=&Multiply_VectorByFloat(Vector=<0.25 0.15 0.2>, Float=0.8)
     )
     Alpha=0.5
    )) {enabled=false, isSetParameters=true, module=SetVariables_#8}
   /SetVariables(Particles.PositionFreeze:NiagaraPosition=&Lerp_Vector(
     A=Linked() {linkedVariable {name=Particles.Position, type=NiagaraPosition}}
     B=Linked() {linkedVariable {name=Particles.PositionHeart, type=NiagaraPosition}}
     Alpha=1
    )) {enabled=false, isSetParameters=true, module=SetVariables_#9}
   &SolveForcesAndVelocity(Clamp Velocity=false, Limit Acceleration=false)
   &ScaleMeshSize(Scale Factor=&VectorFromFloat(Value=&FloatFromCurve(FloatCurve=DataInterface() {propertyValues {Curve {
         defaultValue=0
         keys [
          {interpMode=RCIM_Cubic, tangentMode=RCTM_User, time=0.3003, value=1}
          {arriveTangent=-7.9947, interpMode=RCIM_Cubic, leaveTangent=-7.9947, tangentMode=RCTM_User, time=0.6997, value=0}
         ]
        }}})))
   &ScaleColor(
    Scale RGB=&VectorFromFloat(Value=&MaskFloat(
      Mask Source=1
      Mask Parameter=$Particles.PartitionParticles.Partition
      Mask=0
      Masked Float=1
      Passthrough Float=&MaskFloat(
       Mask Source=1
       Mask Parameter=$Particles.PartitionParticles.Partition
       Mask=1
       Masked Float=1
       Passthrough Float=&MaskFloat(
        Mask Source=1
        Mask Parameter=$Particles.PartitionParticles.Partition
        Mask=2
        Masked Float=1
        Passthrough Float=&MaskFloat(
         Mask Source=1
         Mask Parameter=$Particles.PartitionParticles.Partition
         Mask=3
         Masked Float=1
         Passthrough Float=&MaskFloat(
          Mask Source=1
          Mask Parameter=$Particles.PartitionParticles.Partition
          Mask=4
          Masked Float=1
          Passthrough Float=&MaskFloat(
           Mask Source=1
           Mask Parameter=$Particles.PartitionParticles.Partition
           Mask=5
           Masked Float=1
           Passthrough Float=&MaskFloat(
            Mask Source=1
            Mask Parameter=$Particles.PartitionParticles.Partition
            Mask=6
            Masked Float=1
            Passthrough Float=&MaskFloat(
             Mask Source=1
             Mask Parameter=$Particles.PartitionParticles.Partition
             Mask=7
             Masked Float=1
             Passthrough Float=&MaskFloat(
              Mask Source=1
              Mask Parameter=$Particles.PartitionParticles.Partition
              Mask=8
              Masked Float=1
              Passthrough Float=&MaskFloat(
               Mask Source=1
               Mask Parameter=$Particles.PartitionParticles.Partition
               Mask=9
               Masked Float=1
               Passthrough Float=&MaskFloat(
                Mask Source=1
                Mask Parameter=$Particles.PartitionParticles.Partition
                Mask=10
                Masked Float=1
                Passthrough Float=&MaskFloat(
                 Mask Source=1
                 Mask Parameter=$Particles.PartitionParticles.Partition
                 Mask=11
                 Masked Float=1
                 Passthrough Float=&MaskFloat(
                  Mask Source=1
                  Mask Parameter=$Particles.PartitionParticles.Partition
                  Mask=12
                  Masked Float=1
                  Passthrough Float=&MaskFloat(
                   Mask Source=1
                   Mask Parameter=$Particles.PartitionParticles.Partition
                   Mask=13
                   Masked Float=1
                   Passthrough Float=&MaskFloat(Mask Source=1, Mask Parameter=$Particles.PartitionParticles.Partition, Mask=14, Masked Float=1, Passthrough Float=1)
                  )
                 )
                )
               )
              )
             )
            )
           )
          )
         )
        )
       )
      )
     ))
    ScaleA=true
    ScaleRGB=true
   ) {enabled=false, module=ScaleColor001}
   &UpdateMeshOrientation(
    Rotation Rate=&Multiply_Float(
     A=0
     B=&FloatFromCurve(
      FloatCurve=DataInterface() {propertyValues {Curve {
         defaultValue=1
         keys [
          {arriveTangent=6.9016, interpMode=RCIM_Cubic, leaveTangent=6.9016, tangentMode=RCTM_User, time=1.1108, value=0}
          {interpMode=RCIM_Cubic, tangentMode=RCTM_User, time=1.3484, value=1}
         ]
        }}}
      CurveIndex=$Emitter.Age
     )
    )
    Rotation Vector=<0 0 0.5>
   )
   &ScaleSpriteSize(Uniform Curve Sprite Scale=DataInterface() {propertyValues {Curve {keys [
        {interpMode=RCIM_Cubic, tangentMode=RCTM_User, time=0, value=0.2}
        {interpMode=RCIM_Cubic, tangentMode=RCTM_User, time=0.2, value=1}
        {tangentMode=RCTM_User, time=1, value=0}
       ]}}})
   &DynamicMaterialParameters(
    Index 0 Param 1=0.15
    Index 0 Param 2=&RandomRangeFloat(Minimum=0.25, Maximum=2)
    Index 0 Param 3=0.5
    Module.Index0Param0Write=true
    Module.Index0Param1Write=true
    Module.Index0Param2Write=true
    Module.Index0Param3Write=true
   )
   /SetVariables(
    Particles.Linear Color=&Multiply_LinearColorByFloat(Linear Color=$Particles.Color, Float=2)
    Particles.Vector 2D:Vector2f=&Multiply_Vector2D_ByFloat(Float=&RandomRangeFloat(Minimum=0.15, Maximum=0.25), Vector2D=$Particles.SpriteSize)
   ) {isSetParameters=true, module=SetVariables_#10}
   /SetVariables(
    Particles.UpdateRoundGlowSize:Vector2f=&Multiply_Vector2D(
     A=$Particles.InitialRoundGlowSize
     B=&Vector2DFromFloat(Value=&FloatFromCurve(FloatCurve=DataInterface() {propertyValues {Curve {
          defaultValue=1
          keys [
           {arriveTangent=4.7075, interpMode=RCIM_Cubic, leaveTangent=4.7075, tangentMode=RCTM_User, time=0, value=0}
           {arriveTangent=-1.2993, interpMode=RCIM_Cubic, leaveTangent=-1.2993, tangentMode=RCTM_User, time=0.4887, value=0.9847}
           {arriveTangent=-0.4432, interpMode=RCIM_Cubic, leaveTangent=-0.4432, tangentMode=RCTM_User, time=1, value=0}
          ]
         }}}))
    )
    Particles.UpdateStarSize:Vector2f=&Multiply_Vector2D(
     A=$Particles.InitialStarSize
     B=&Vector2DFromCurve(
      Vector2Curve=DataInterface() {propertyValues {
        XCurve {
         defaultValue=0
         keys [
          {arriveTangent=0.0865, interpMode=RCIM_Cubic, leaveTangent=0.0865, tangentMode=RCTM_User, time=0, value=0}
          {arriveTangent=4.9825, interpMode=RCIM_Cubic, leaveTangent=4.9825, tangentMode=RCTM_User, time=0.4724, value=1}
          {arriveTangent=-1.9029, interpMode=RCIM_Cubic, leaveTangent=-1.9029, time=0.8093, value=1.9173}
          <at 0.9979 0 RCIM_Cubic>
         ]
        }
        YCurve {
         defaultValue=1
         keys [
          {arriveTangent=-5.5781, interpMode=RCIM_Cubic, leaveTangent=-5.5781, tangentMode=RCTM_User, time=0.6, value=1.1}
          {arriveTangent=-0.1643, interpMode=RCIM_Cubic, leaveTangent=-0.1643, tangentMode=RCTM_User, time=1, value=0}
         ]
        }
       }}
      ScaleCurve=&RandomRangeVector2D(Minimum=<1 1>, Maximum=<1 1.2>)
     )
    )
   ) {isSetParameters=true, module=SetVariables_#11}
   /SetVariables(
    Particles.CameraOffsetStar=15
    Particles.GlowColor=&MakeLinearColorFromVectorAndFloat(
     "Vector (RGB)"=&Multiply_Vector(A=&MakeVectorFromLinearColorRGB(Color=<rgba 1.2 0.6795 0 1>), B=<1 1 1>)
     "Float (Alpha)"=1
    )
    Particles.GlowSize=&Multiply_Vector2D_ByFloat(Float=1.75, Vector2D=$Particles.UpdateStarSize)
    Particles.RoundGlow:LinearColor=&MakeLinearColorFromVectorAndFloat(
     "Vector (RGB)"=&Multiply_Vector(A=&MakeVectorFromLinearColorRGB(Color=<rgba 1.2 0.6795 0 1>), B=<1 1 1>)
     "Float (Alpha)"=0.15
    )
   ) {isSetParameters=true, module=SetVariables_#12}
   /SetVariables(Particles.Position3DFaceCamera:NiagaraPosition=&SubtractVectorFromPosition(
     Position=&AddVectorToPosition(
      Position=&Multiply_VectorByFloat(
       Vector=&ReturnCameraVector(Camera Vectors=::NewEnumerator3, Coordinate Space=::Simulation)
       Float=&MakeFloatFromVector(Vector=Linked() {linkedVariable {name=Particles.Position, type=NiagaraPosition}}, Channel=::NewEnumerator0)
      )
      Vector=&AddVectorToPosition(
       Position=&Multiply_VectorByFloat(
        Vector=&ReturnCameraVector(Camera Vectors=::NewEnumerator2, Coordinate Space=::Simulation)
        Float=&MakeFloatFromVector(Vector=Linked() {linkedVariable {name=Particles.Position, type=NiagaraPosition}}, Channel=::NewEnumerator1)
       )
       Vector=&Multiply_VectorByFloat(
        Vector=&ReturnCameraVector(Camera Vectors=::NewEnumerator2, Coordinate Space=::Simulation)
        Float=&MakeFloatFromVector(Vector=Linked() {linkedVariable {name=Particles.Position, type=NiagaraPosition}}, Channel=::NewEnumerator2)
       )
      )
     )
     Vector=Linked() {linkedVariable {name=Engine.Emitter.SimulationPosition, type=NiagaraPosition}}
    )) {isSetParameters=true, module=SetVariables_#13}
  ]
  rendererClasses [&NiagaraMeshRendererProperties, &NiagaraSpriteRendererProperties]
  renderers [
   {rendererClass=&NiagaraMeshRendererProperties, rendererIndex=0}
   {rendererClass=&NiagaraMeshRendererProperties, rendererIndex=1}
   {rendererClass=&NiagaraSpriteRendererProperties, rendererIndex=2}
   {rendererClass=&NiagaraSpriteRendererProperties, rendererIndex=3}
   {rendererClass=&NiagaraSpriteRendererProperties, rendererIndex=4}
   {rendererClass=&NiagaraSpriteRendererProperties, rendererIndex=5}
   {rendererClass=&NiagaraSpriteRendererProperties, rendererIndex=6}
  ]
 }
 Pop_FaceCamera001 {
  bEnabled=false
  emitterUpdateScript [
   &EmitterState()
   &SpawnBurst_Instantaneous(
    Spawn Count=16
    Spawn Time=&Subtract_Float(A=$System.IntroLength, B=0)
    Use Spawn Probability=false
    UseLoopCountLimit=false
   )
   &SpawnBurst_Instantaneous(
    Spawn Count=&RandomRangeInt(Minimum=30, Maximum=50)
    Spawn Group=1
    Spawn Time=&Subtract_Float(A=$System.IntroLength, B=0.03)
    Use Spawn Probability=false
    UseLoopCountLimit=false
   ) {module=SpawnBurst_Instantaneous001}
   &SpawnBurst_Instantaneous(
    Spawn Group=2
    Spawn Time=&Subtract_Float(A=$System.IntroLength, B=0.03)
    Use Spawn Probability=false
    UseLoopCountLimit=false
   ) {module=SpawnBurst_Instantaneous002}
   &SpawnBurst_Instantaneous(
    Spawn Group=3
    Spawn Time=&Subtract_Float(A=$System.IntroLength, B=0.03)
    Use Spawn Probability=false
    UseLoopCountLimit=false
   ) {module=SpawnBurst_Instantaneous003}
   &SpawnBurst_Instantaneous(
    Spawn Group=3
    Spawn Time=&Subtract_Float(A=$System.IntroLength, B=0.03)
    Use Spawn Probability=false
    UseLoopCountLimit=false
   ) {module=SpawnBurst_Instantaneous004}
  ]
  particleSpawnScript [
   /SetVariables(Particles.VisibilityTag=&SpawnGroup()) {isSetParameters=true, module=SetVariables_#6}
   &PartitionParticles(Number Of Partitions=16, Particles Per Partition=1, Partition Mode:ENiagara_PartitionMode_v2=::NewEnumerator1)
   &InitializeParticle(
    Color=&MergeHelper_ColorSwitch(
     0=&FNTimeOfDayColor(
      Night Color=<rgba 2.3125 1.1562 0 1>
      Morning Color=<rgba 2.3125 1.2719 0 1>
      Day Color=<rgba 2.3125 1.1562 0 1>
      Evening Color=<rgba 2.3125 0.9635 0 1>
     )
     1=&RandomRangeLinearColor(Minimum=<rgba 1 0.5281 0.0563 0.5>, Maximum=<rgba 1 0.45 0 0.5>)
     2=<rgba 3 1.8435 0.2244 1>
     3=<rgba 3 1.8438 0.225 1>
     Index=$Particles.VisibilityTag
     Num=::NewEnumerator2
    )
    Lifetime=&MergeHelper_FloatSwitch(
     0=&RandomRangeFloat(Minimum=0.75, Maximum=1.2, Randomness Mode=::NewEnumerator2, Random Seed=?)
     1=&RandomRangeFloat(Minimum=0.4, Maximum=1.25)
     2=0.1
     3=0.1
     Index=$Particles.VisibilityTag
     Num=::NewEnumerator2
    )
    Mesh Scale Mode=::NewEnumerator3
    Mesh Uniform Scale=0.25
    Position Offset=&MergeHelper_VectorSwitch(
     0=<0 0 0>
     1=<0 0 0>
     2=&AddVectorToPosition(Position=Linked() {linkedVariable {name=System.BurstPosition, type=Vector3f}}, Vector=<0 0 0>)
     3=&AddVectorToPosition(Position=Linked() {linkedVariable {name=System.BurstPosition, type=Vector3f}}, Vector=<0 0 0>)
     Index=$Particles.VisibilityTag
     Num=::NewEnumerator2
    )
    Sprite Rotation Angle=-20
    Sprite Rotation Mode=::NewEnumerator0
    Sprite Size Mode=::NewEnumerator3
    Uniform Sprite Size=&Lerp_Float(
     A=10
     B=50
     Alpha=&FloatFromCurve(
      FloatCurve=DataInterface() {propertyValues {Curve {keys [
          {interpMode=RCIM_Cubic, tangentMode=RCTM_User, time=0, value=0.2}
          {arriveTangent=1.1314, interpMode=RCIM_Cubic, leaveTangent=1.1314, tangentMode=RCTM_User, time=0.9, value=0.6}
          {interpMode=RCIM_Cubic, tangentMode=RCTM_User, time=1, value=1}
         ]}}}
      CurveIndex=&RandomRangeFloat(Minimum=0, Maximum=1)
     )
    )
    UsePositionOffset=true
    Write Material Random=true
   )
   /SetVariables(
    Particles.GlitterSubUV:NiagaraFloat=&RandomRangeFloat(Minimum=0, Maximum=3)
    Particles.InitialRoundGlowSize:Vector2f=<100 100>
    Particles.InitialStarSize:Vector2f=&Multiply_Vector2D_ByFloat(Float=0.6, Vector2D=<30 40>)
   ) {isSetParameters=true, module=SetVariables_#7}
   &ShapeLocation(
    MaskSpawnGroup=false
    Offset=&MergeHelper_VectorSwitch(0=$System.BurstPosition, 1=$System.BurstPosition, 2=$System.BurstPosition, 3=?, Index=$Particles.VisibilityTag, Num=::NewEnumerator1)
    Offset Mode=::NewEnumerator0
    Sphere Distribution=::NewEnumerator3
    Sphere Radius=&MergeHelper_FloatSwitch(0=2, 1=1, 2=?, 3=?, Index=$Particles.VisibilityTag, Num=::NewEnumerator0)
    Uniform Particle Index:NiagaraInt32=&MergeHelper_IntSwitch(
     0=$Particles.PartitionParticles.Partition
     1=&ReturnExecIndex()
     2=0
     3=0
     4=?
     5=?
     Index=$Particles.VisibilityTag
     Num=::NewEnumerator2
    )
    Uniform Spiral Amount=0.5
    UseOwnerScale=true
   )
   &AddVelocity(
    Module.Use Falloff Distance=false
    Origin Offset=&MergeHelper_VectorSwitch(
     0=&AddVectorToPosition(Position=<0 0 -4>, Vector=$System.BurstPosition)
     1=&AddVectorToPosition(Position=<0 0 -0.6>, Vector=$System.BurstPosition)
     2=?
     3=?
     Index=$Particles.VisibilityTag
     Num=::NewEnumerator0
    )
    Override Seed=false
    Randomness Mode:ENiagaraRandomnessMode=::NewEnumerator2
    Velocity Mode=::NewEnumerator1
    Velocity Speed=&MergeHelper_FloatSwitch(
     0=0
     1=&RandomRangeFloat(
      Minimum=30
      Maximum=&FN_FrontEndFloatModifier(Gameplay=300, Front End=200)
      Randomness Mode=::NewEnumerator2
      Random Seed=?
     )
     2=?
     3=?
     Index=$Particles.VisibilityTag
     Num=::NewEnumerator0
    )
   )
   &AddVelocity(Module.Use Falloff Distance=false, Override Seed=false, Velocity Mode=::NewEnumerator1) {enabled=false, module=AddVelocity001}
   &InitialMeshOrientation(Mesh Axis To Orient=<0 1 0>, Mesh Orientation Mode=::NewEnumerator2, Use Rotation Vector=false)
  ]
  particleUpdateScript [
   &ParticleState()
   &ScaleColor(
    Scale Alpha=&MergeHelper_FloatSwitch(
     0=1
     1=&Multiply_Float(
      A=0.25
      B=&FloatFromCurve(
       FloatCurve=DataInterface() {propertyValues {Curve {keys [<at 0 1>,<at 1 0>]}}}
       Scale Curve=&Waveform(
        Global Amplitude Scale=10
        "Waveform [1]"=::NewEnumerator0
        "[1] Phase Scale"=&RandomRangeFloat(Minimum=2, Maximum=4)
        "[1] Frequency"=&RandomRangeFloat(Minimum=0.75, Maximum=1)
        "[1] Amplitude Min/Max"=<0 1>
       )
      )
     )
     2=1
     3=1
     Index=$Particles.VisibilityTag
     Num=::NewEnumerator2
    )
    Scale RGB=&MergeHelper_VectorSwitch(
     0=&FN_FrontEndVector3Modifier(
      Gameplay=&FN_Select_Vector_from_Time_Of_Day(Morning=<1.115 1.15 1.15>, Day=<1 1 1>, Evening=<1 1 1>, Night=<2.65 2.65 2.65>)
      Front End=<3.2 2.75 2.75>
     )
     1=<1 1 1>
     2=<1 1 1>
     3=<1 1 1>
     Index=0
     Num=::NewEnumerator2
    )
    ScaleA=true
    ScaleRGB=true
   )
   &GravityForce(Gravity=&Multiply_VectorByFloat(
     Vector=&MergeHelper_VectorSwitch(
      0=&RandomRangeVector(Minimum=<0 0 -80>, Maximum=<0 0 -110>, Randomness Mode=::NewEnumerator1, Random Seed=10, Fixed Random Seed=true)
      1=&RandomRangeVector(Minimum=<0 0 -250>, Maximum=<0 0 -50>)
      2=?
      3=?
      Index=$Particles.VisibilityTag
      Num=::NewEnumerator0
     )
     Float=&MergeHelper_FloatSwitch(
      0=&FloatFromCurve(
       FloatCurve=DataInterface() {propertyValues {Curve {keys [{arriveTangent=1.7842, interpMode=RCIM_Cubic, leaveTangent=1.7842, tangentMode=RCTM_User, time=0, value=1}, <at 1 2 RCIM_Cubic>]}}}
       Scale Curve=1.6
      )
      1=1
      2=?
      3=?
      Index=$Particles.VisibilityTag
      Num=::NewEnumerator0
     )
    ))
   &Drag(
    Drag=&MergeHelper_FloatSwitch(
     0=&FloatFromCurve(
      FloatCurve=DataInterface() {propertyValues {Curve {keys [<at 0 0 RCIM_Cubic>,<at 0.176 0.7545 RCIM_Cubic>,<at 0.3 0 RCIM_Cubic>]}}}
      Scale Curve=3
     )
     1=&RandomRangeFloat(Minimum=5, Maximum=12)
     2=?
     3=?
     Index=$Particles.VisibilityTag
     Num=::NewEnumerator0
    )
    Use Linear Drag=true
    Use Rotational Drag=true
   )
   &PointAttractionForce(
    Attraction Radius=15
    Attraction Strength=&MergeHelper_FloatSwitch(
     0=&FloatFromCurve(
      FloatCurve=DataInterface() {propertyValues {Curve {keys [
          {arriveTangent=36.0485, interpMode=RCIM_Cubic, leaveTangent=36.0485, tangentMode=RCTM_User, time=0, value=0}
          {arriveTangent=0.4397, interpMode=RCIM_Cubic, leaveTangent=0.4397, tangentMode=RCTM_User, time=0.089, value=1}
          {arriveTangent=-0.6185, interpMode=RCIM_Cubic, leaveTangent=-0.6185, tangentMode=RCTM_User, time=0.3, value=0}
         ]}}}
      Scale Curve=100
     )
     1=0
     2=0
     3=0
     Index=$Particles.VisibilityTag
     Num=::NewEnumerator2
    )
    Attractor Position Offset:Vector3f=&AddVectorToPosition(
     Position=&Multiply_Vector(
      A=&MaskPosition(
       Mask Source:NiagaraInt32=1
       Mask Parameter:NiagaraInt32=$Particles.PartitionParticles.Partition
       Mask:NiagaraInt32=0
       Masked Position:NiagaraPosition=<-48 0 17>
       Passthrough Position:NiagaraPosition=&MaskPosition(
        Mask Source:NiagaraInt32=1
        Mask Parameter:NiagaraInt32=$Particles.PartitionParticles.Partition
        Mask:NiagaraInt32=1
        Masked Position:NiagaraPosition=<-40 0 -29>
        Passthrough Position:NiagaraPosition=&MaskPosition(
         Mask Source:NiagaraInt32=1
         Mask Parameter:NiagaraInt32=$Particles.PartitionParticles.Partition
         Mask:NiagaraInt32=2
         Masked Position:NiagaraPosition=<40 0 -29>
         Passthrough Position:NiagaraPosition=&MaskPosition(
          Mask Source:NiagaraInt32=1
          Mask Parameter:NiagaraInt32=$Particles.PartitionParticles.Partition
          Mask:NiagaraInt32=3
          Masked Position:NiagaraPosition=<-25 0 -47>
          Passthrough Position:NiagaraPosition=&MaskPosition(
           Mask Source:NiagaraInt32=1
           Mask Parameter:NiagaraInt32=$Particles.PartitionParticles.Partition
           Mask:NiagaraInt32=4
           Masked Position:NiagaraPosition=<0 0 -92>
           Passthrough Position:NiagaraPosition=&MaskPosition(
            Mask Source:NiagaraInt32=1
            Mask Parameter:NiagaraInt32=$Particles.PartitionParticles.Partition
            Mask:NiagaraInt32=5
            Masked Position:NiagaraPosition=<14 0 10>
            Passthrough Position:NiagaraPosition=&MaskPosition(
             Mask Source:NiagaraInt32=1
             Mask Parameter:NiagaraInt32=$Particles.PartitionParticles.Partition
             Mask:NiagaraInt32=6
             Masked Position:NiagaraPosition=<0 0 -10>
             Passthrough Position:NiagaraPosition=&MaskPosition(
              Mask Source:NiagaraInt32=1
              Mask Parameter:NiagaraInt32=$Particles.PartitionParticles.Partition
              Mask:NiagaraInt32=7
              Masked Position:NiagaraPosition=<-12 0 10>
              Passthrough Position:NiagaraPosition=&MaskPosition(
               Mask Source:NiagaraInt32=1
               Mask Parameter:NiagaraInt32=$Particles.PartitionParticles.Partition
               Mask:NiagaraInt32=8
               Masked Position:NiagaraPosition=<48 0 17>
               Passthrough Position:NiagaraPosition=&MaskPosition(
                Mask Source:NiagaraInt32=1
                Mask Parameter:NiagaraInt32=$Particles.PartitionParticles.Partition
                Mask:NiagaraInt32=9
                Masked Position:NiagaraPosition=<28 0 28>
                Passthrough Position:NiagaraPosition=&MaskPosition(
                 Mask Source:NiagaraInt32=1
                 Mask Parameter:NiagaraInt32=$Particles.PartitionParticles.Partition
                 Mask:NiagaraInt32=10
                 Masked Position:NiagaraPosition=<-28 0 28>
                 Passthrough Position:NiagaraPosition=&MaskPosition(
                  Mask Source:NiagaraInt32=1
                  Mask Parameter:NiagaraInt32=$Particles.PartitionParticles.Partition
                  Mask:NiagaraInt32=11
                  Masked Position:NiagaraPosition=<50 0 -7>
                  Passthrough Position:NiagaraPosition=&MaskPosition(
                   Mask Source:NiagaraInt32=1
                   Mask Parameter:NiagaraInt32=$Particles.PartitionParticles.Partition
                   Mask:NiagaraInt32=12
                   Masked Position:NiagaraPosition=<12 0 -70>
                   Passthrough Position:NiagaraPosition=&MaskPosition(
                    Mask Source:NiagaraInt32=1
                    Mask Parameter:NiagaraInt32=$Particles.PartitionParticles.Partition
                    Mask:NiagaraInt32=13
                    Masked Position:NiagaraPosition=<-50 0 -7>
                    Passthrough Position:NiagaraPosition=&MaskVector(Mask Source=1, Mask Parameter=$Particles.PartitionParticles.Partition, Mask=14, Masked Vector=<-12 0 -70>, Passthrough Vector=<25 0 -47>)
                   )
                  )
                 )
                )
               )
              )
             )
            )
           )
          )
         )
        )
       )
      )
      B=&Multiply_VectorByFloat(Vector=<0.25 0.15 0.2>, Float=&FN_FrontEndFloatModifier(Gameplay=0.55, Front End=0.4))
     )
     Vector=&AddVectorToPosition(
      Position=Linked() {linkedVariable {name=System.BurstPosition, type=Vector3f}}
      Vector=&VectorFromCurve(
       VectorCurve=DataInterface() {propertyValues {XCurve {keys [<at 0 0>,<at 1 0>]}, YCurve {keys [<at 0 0>,<at 1 0>]}, ZCurve {keys [<at 0 1 RCIM_Cubic>,<at 0.7003 -1 RCIM_Cubic>]}}}
       Scale Curve=<1 1 5>
      )
     )
    )
    Attractor Position Offset Coordinate Space:ENiagaraCoordinateSpace=::Local
    Falloff Exponent=1
    Kill Within Radius=false
    Module.Use Falloff=true
    UseAttractorOffset=true
   )
   /SetVariables(Particles.PositionHeart:NiagaraPosition=&Lerp_Vector(
     A=Linked() {linkedVariable {name=Particles.Position, type=NiagaraPosition}}
     B=&Multiply_Vector(
      A=&AddVectorToPosition(
       Position=&MaskPosition(
        Mask Source:NiagaraInt32=1
        Mask Parameter:NiagaraInt32=$Particles.PartitionParticles.Partition
        Mask:NiagaraInt32=0
        Masked Position:NiagaraPosition=<-48 0 17>
        Passthrough Position:NiagaraPosition=&MaskPosition(
         Mask Source:NiagaraInt32=1
         Mask Parameter:NiagaraInt32=$Particles.PartitionParticles.Partition
         Mask:NiagaraInt32=1
         Masked Position:NiagaraPosition=<-40 0 -29>
         Passthrough Position:NiagaraPosition=&MaskPosition(
          Mask Source:NiagaraInt32=1
          Mask Parameter:NiagaraInt32=$Particles.PartitionParticles.Partition
          Mask:NiagaraInt32=2
          Masked Position:NiagaraPosition=<40 0 -29>
          Passthrough Position:NiagaraPosition=&MaskPosition(
           Mask Source:NiagaraInt32=1
           Mask Parameter:NiagaraInt32=$Particles.PartitionParticles.Partition
           Mask:NiagaraInt32=3
           Masked Position:NiagaraPosition=<-25 0 -47>
           Passthrough Position:NiagaraPosition=&MaskPosition(
            Mask Source:NiagaraInt32=1
            Mask Parameter:NiagaraInt32=$Particles.PartitionParticles.Partition
            Mask:NiagaraInt32=4
            Masked Position:NiagaraPosition=<0 0 -92>
            Passthrough Position:NiagaraPosition=&MaskPosition(
             Mask Source:NiagaraInt32=1
             Mask Parameter:NiagaraInt32=$Particles.PartitionParticles.Partition
             Mask:NiagaraInt32=5
             Masked Position:NiagaraPosition=<14 0 10>
             Passthrough Position:NiagaraPosition=&MaskPosition(
              Mask Source:NiagaraInt32=1
              Mask Parameter:NiagaraInt32=$Particles.PartitionParticles.Partition
              Mask:NiagaraInt32=6
              Masked Position:NiagaraPosition=<0 0 -10>
              Passthrough Position:NiagaraPosition=&MaskPosition(
               Mask Source:NiagaraInt32=1
               Mask Parameter:NiagaraInt32=$Particles.PartitionParticles.Partition
               Mask:NiagaraInt32=7
               Masked Position:NiagaraPosition=<-12 0 10>
               Passthrough Position:NiagaraPosition=&MaskPosition(
                Mask Source:NiagaraInt32=1
                Mask Parameter:NiagaraInt32=$Particles.PartitionParticles.Partition
                Mask:NiagaraInt32=8
                Masked Position:NiagaraPosition=<48 0 17>
                Passthrough Position:NiagaraPosition=&MaskPosition(
                 Mask Source:NiagaraInt32=1
                 Mask Parameter:NiagaraInt32=$Particles.PartitionParticles.Partition
                 Mask:NiagaraInt32=9
                 Masked Position:NiagaraPosition=<28 0 28>
                 Passthrough Position:NiagaraPosition=&MaskPosition(
                  Mask Source:NiagaraInt32=1
                  Mask Parameter:NiagaraInt32=$Particles.PartitionParticles.Partition
                  Mask:NiagaraInt32=10
                  Masked Position:NiagaraPosition=<-28 0 28>
                  Passthrough Position:NiagaraPosition=&MaskPosition(
                   Mask Source:NiagaraInt32=1
                   Mask Parameter:NiagaraInt32=$Particles.PartitionParticles.Partition
                   Mask:NiagaraInt32=11
                   Masked Position:NiagaraPosition=<50 0 -7>
                   Passthrough Position:NiagaraPosition=&MaskPosition(
                    Mask Source:NiagaraInt32=1
                    Mask Parameter:NiagaraInt32=$Particles.PartitionParticles.Partition
                    Mask:NiagaraInt32=12
                    Masked Position:NiagaraPosition=<12 0 -70>
                    Passthrough Position:NiagaraPosition=&MaskPosition(
                     Mask Source:NiagaraInt32=1
                     Mask Parameter:NiagaraInt32=$Particles.PartitionParticles.Partition
                     Mask:NiagaraInt32=13
                     Masked Position:NiagaraPosition=<-50 0 -7>
                     Passthrough Position:NiagaraPosition=&MaskVector(Mask Source=1, Mask Parameter=$Particles.PartitionParticles.Partition, Mask=14, Masked Vector=<-12 0 -70>, Passthrough Vector=<25 0 -47>)
                    )
                   )
                  )
                 )
                )
               )
              )
             )
            )
           )
          )
         )
        )
       )
       Vector=<0 0 20>
      )
      B=&Multiply_VectorByFloat(Vector=<0.25 0.15 0.2>, Float=0.8)
     )
     Alpha=0.5
    )) {enabled=false, isSetParameters=true, module=SetVariables_#8}
   /SetVariables(Particles.PositionFreeze:NiagaraPosition=&Lerp_Vector(
     A=Linked() {linkedVariable {name=Particles.Position, type=NiagaraPosition}}
     B=Linked() {linkedVariable {name=Particles.PositionHeart, type=NiagaraPosition}}
     Alpha=1
    )) {enabled=false, isSetParameters=true, module=SetVariables_#9}
   &SolveForcesAndVelocity(Clamp Velocity=false, Limit Acceleration=false)
   &ScaleMeshSize(Scale Factor=&VectorFromFloat(Value=&FloatFromCurve(FloatCurve=DataInterface() {propertyValues {Curve {
         defaultValue=0
         keys [
          {interpMode=RCIM_Cubic, tangentMode=RCTM_User, time=0.3003, value=1}
          {arriveTangent=-7.9947, interpMode=RCIM_Cubic, leaveTangent=-7.9947, tangentMode=RCTM_User, time=0.6997, value=0}
         ]
        }}})))
   &ScaleColor(
    Scale RGB=&VectorFromFloat(Value=&MaskFloat(
      Mask Source=1
      Mask Parameter=$Particles.PartitionParticles.Partition
      Mask=0
      Masked Float=1
      Passthrough Float=&MaskFloat(
       Mask Source=1
       Mask Parameter=$Particles.PartitionParticles.Partition
       Mask=1
       Masked Float=1
       Passthrough Float=&MaskFloat(
        Mask Source=1
        Mask Parameter=$Particles.PartitionParticles.Partition
        Mask=2
        Masked Float=1
        Passthrough Float=&MaskFloat(
         Mask Source=1
         Mask Parameter=$Particles.PartitionParticles.Partition
         Mask=3
         Masked Float=1
         Passthrough Float=&MaskFloat(
          Mask Source=1
          Mask Parameter=$Particles.PartitionParticles.Partition
          Mask=4
          Masked Float=1
          Passthrough Float=&MaskFloat(
           Mask Source=1
           Mask Parameter=$Particles.PartitionParticles.Partition
           Mask=5
           Masked Float=1
           Passthrough Float=&MaskFloat(
            Mask Source=1
            Mask Parameter=$Particles.PartitionParticles.Partition
            Mask=6
            Masked Float=1
            Passthrough Float=&MaskFloat(
             Mask Source=1
             Mask Parameter=$Particles.PartitionParticles.Partition
             Mask=7
             Masked Float=1
             Passthrough Float=&MaskFloat(
              Mask Source=1
              Mask Parameter=$Particles.PartitionParticles.Partition
              Mask=8
              Masked Float=1
              Passthrough Float=&MaskFloat(
               Mask Source=1
               Mask Parameter=$Particles.PartitionParticles.Partition
               Mask=9
               Masked Float=1
               Passthrough Float=&MaskFloat(
                Mask Source=1
                Mask Parameter=$Particles.PartitionParticles.Partition
                Mask=10
                Masked Float=1
                Passthrough Float=&MaskFloat(
                 Mask Source=1
                 Mask Parameter=$Particles.PartitionParticles.Partition
                 Mask=11
                 Masked Float=1
                 Passthrough Float=&MaskFloat(
                  Mask Source=1
                  Mask Parameter=$Particles.PartitionParticles.Partition
                  Mask=12
                  Masked Float=1
                  Passthrough Float=&MaskFloat(
                   Mask Source=1
                   Mask Parameter=$Particles.PartitionParticles.Partition
                   Mask=13
                   Masked Float=1
                   Passthrough Float=&MaskFloat(Mask Source=1, Mask Parameter=$Particles.PartitionParticles.Partition, Mask=14, Masked Float=1, Passthrough Float=1)
                  )
                 )
                )
               )
              )
             )
            )
           )
          )
         )
        )
       )
      )
     ))
    ScaleA=true
    ScaleRGB=true
   ) {enabled=false, module=ScaleColor001}
   &UpdateMeshOrientation(
    Rotation Rate=&Multiply_Float(
     A=0
     B=&FloatFromCurve(
      FloatCurve=DataInterface() {propertyValues {Curve {
         defaultValue=1
         keys [
          {arriveTangent=6.9016, interpMode=RCIM_Cubic, leaveTangent=6.9016, tangentMode=RCTM_User, time=1.1108, value=0}
          {interpMode=RCIM_Cubic, tangentMode=RCTM_User, time=1.3484, value=1}
         ]
        }}}
      CurveIndex=$Emitter.Age
     )
    )
    Rotation Vector=<0 0 0.5>
   )
   &ScaleSpriteSize(Uniform Curve Sprite Scale=DataInterface() {propertyValues {Curve {keys [
        {interpMode=RCIM_Cubic, tangentMode=RCTM_User, time=0, value=0.2}
        {interpMode=RCIM_Cubic, tangentMode=RCTM_User, time=0.2, value=1}
        {tangentMode=RCTM_User, time=1, value=0}
       ]}}})
   &DynamicMaterialParameters(
    Index 0 Param 1=0.15
    Index 0 Param 2=&RandomRangeFloat(Minimum=0.25, Maximum=2)
    Index 0 Param 3=0.5
    Module.Index0Param0Write=true
    Module.Index0Param1Write=true
    Module.Index0Param2Write=true
    Module.Index0Param3Write=true
   )
   /SetVariables(
    Particles.Linear Color=&Multiply_LinearColorByFloat(Linear Color=$Particles.Color, Float=2)
    Particles.Vector 2D:Vector2f=&Multiply_Vector2D_ByFloat(Float=&RandomRangeFloat(Minimum=0.15, Maximum=0.25), Vector2D=$Particles.SpriteSize)
   ) {isSetParameters=true, module=SetVariables_#10}
   /SetVariables(
    Particles.UpdateRoundGlowSize:Vector2f=&Multiply_Vector2D(
     A=$Particles.InitialRoundGlowSize
     B=&Vector2DFromFloat(Value=&FloatFromCurve(FloatCurve=DataInterface() {propertyValues {Curve {
          defaultValue=1
          keys [
           {arriveTangent=4.7075, interpMode=RCIM_Cubic, leaveTangent=4.7075, tangentMode=RCTM_User, time=0, value=0}
           {arriveTangent=-1.2993, interpMode=RCIM_Cubic, leaveTangent=-1.2993, tangentMode=RCTM_User, time=0.4887, value=0.9847}
           {arriveTangent=-0.4432, interpMode=RCIM_Cubic, leaveTangent=-0.4432, tangentMode=RCTM_User, time=1, value=0}
          ]
         }}}))
    )
    Particles.UpdateStarSize:Vector2f=&Multiply_Vector2D(
     A=$Particles.InitialStarSize
     B=&Vector2DFromCurve(
      Vector2Curve=DataInterface() {propertyValues {
        XCurve {
         defaultValue=0
         keys [
          {arriveTangent=0.0865, interpMode=RCIM_Cubic, leaveTangent=0.0865, tangentMode=RCTM_User, time=0, value=0}
          {arriveTangent=4.9825, interpMode=RCIM_Cubic, leaveTangent=4.9825, tangentMode=RCTM_User, time=0.4724, value=1}
          {arriveTangent=-1.9029, interpMode=RCIM_Cubic, leaveTangent=-1.9029, time=0.8093, value=1.9173}
          <at 0.9979 0 RCIM_Cubic>
         ]
        }
        YCurve {
         defaultValue=1
         keys [
          {arriveTangent=-5.5781, interpMode=RCIM_Cubic, leaveTangent=-5.5781, tangentMode=RCTM_User, time=0.6, value=1.1}
          {arriveTangent=-0.1643, interpMode=RCIM_Cubic, leaveTangent=-0.1643, tangentMode=RCTM_User, time=1, value=0}
         ]
        }
       }}
      ScaleCurve=&RandomRangeVector2D(Minimum=<1 1>, Maximum=<1 1.2>)
     )
    )
   ) {isSetParameters=true, module=SetVariables_#11}
   /SetVariables(
    Particles.CameraOffsetStar=15
    Particles.GlowColor=&MakeLinearColorFromVectorAndFloat(
     "Vector (RGB)"=&Multiply_Vector(A=&MakeVectorFromLinearColorRGB(Color=<rgba 1.2 0.6795 0 1>), B=<1 1 1>)
     "Float (Alpha)"=1
    )
    Particles.GlowSize=&Multiply_Vector2D_ByFloat(Float=1.75, Vector2D=$Particles.UpdateStarSize)
    Particles.RoundGlow:LinearColor=&MakeLinearColorFromVectorAndFloat(
     "Vector (RGB)"=&Multiply_Vector(A=&MakeVectorFromLinearColorRGB(Color=<rgba 1.2 0.6795 0 1>), B=<1 1 1>)
     "Float (Alpha)"=0.15
    )
   ) {isSetParameters=true, module=SetVariables_#12}
   /SetVariables(Particles.Position3DFaceCamera:NiagaraPosition=&SubtractVectorFromPosition(
     Position=&AddVectorToPosition(
      Position=&Multiply_VectorByFloat(
       Vector=&ReturnCameraVector(Camera Vectors=::NewEnumerator3, Coordinate Space=::Simulation)
       Float=&MakeFloatFromVector(Vector=Linked() {linkedVariable {name=Particles.Position, type=NiagaraPosition}}, Channel=::NewEnumerator0)
      )
      Vector=&AddVectorToPosition(
       Position=&Multiply_VectorByFloat(
        Vector=&ReturnCameraVector(Camera Vectors=::NewEnumerator2, Coordinate Space=::Simulation)
        Float=&MakeFloatFromVector(Vector=Linked() {linkedVariable {name=Particles.Position, type=NiagaraPosition}}, Channel=::NewEnumerator1)
       )
       Vector=&Multiply_VectorByFloat(
        Vector=&ReturnCameraVector(Camera Vectors=::NewEnumerator2, Coordinate Space=::Simulation)
        Float=&MakeFloatFromVector(Vector=Linked() {linkedVariable {name=Particles.Position, type=NiagaraPosition}}, Channel=::NewEnumerator2)
       )
      )
     )
     Vector=Linked() {linkedVariable {name=Engine.Emitter.SimulationPosition, type=NiagaraPosition}}
    )) {isSetParameters=true, module=SetVariables_#13}
  ]
  rendererClasses [&NiagaraMeshRendererProperties, &NiagaraSpriteRendererProperties]
  renderers [
   {rendererClass=&NiagaraMeshRendererProperties, rendererIndex=0}
   {rendererClass=&NiagaraMeshRendererProperties, rendererIndex=1}
   {rendererClass=&NiagaraSpriteRendererProperties, rendererIndex=2}
   {rendererClass=&NiagaraSpriteRendererProperties, rendererIndex=3}
   {rendererClass=&NiagaraSpriteRendererProperties, rendererIndex=4}
   {rendererClass=&NiagaraSpriteRendererProperties, rendererIndex=5}
   {rendererClass=&NiagaraSpriteRendererProperties, rendererIndex=6}
  ]
 }
 Pop_V1 {
  bEnabled=false
  emitterUpdateScript [
   &EmitterState()
   &SpawnBurst_Instantaneous(
    Spawn Count=16
    Spawn Time=&Subtract_Float(A=$System.IntroLength, B=0)
    Use Spawn Probability=false
    UseLoopCountLimit=false
   )
   &SpawnBurst_Instantaneous(
    Spawn Count=&RandomRangeInt(Minimum=30, Maximum=50)
    Spawn Group=1
    Spawn Time=&Subtract_Float(A=$System.IntroLength, B=0.03)
    Use Spawn Probability=false
    UseLoopCountLimit=false
   ) {module=SpawnBurst_Instantaneous001}
   &SpawnBurst_Instantaneous(
    Spawn Group=2
    Spawn Time=&Subtract_Float(A=$System.IntroLength, B=0.03)
    Use Spawn Probability=false
    UseLoopCountLimit=false
   ) {module=SpawnBurst_Instantaneous002}
   &SpawnBurst_Instantaneous(
    Spawn Group=3
    Spawn Time=&Subtract_Float(A=$System.IntroLength, B=0.03)
    Use Spawn Probability=false
    UseLoopCountLimit=false
   ) {module=SpawnBurst_Instantaneous003}
   &SpawnBurst_Instantaneous(
    Spawn Group=3
    Spawn Time=&Subtract_Float(A=$System.IntroLength, B=0.03)
    Use Spawn Probability=false
    UseLoopCountLimit=false
   ) {module=SpawnBurst_Instantaneous004}
  ]
  particleSpawnScript [
   /SetVariables(Particles.VisibilityTag=&SpawnGroup()) {isSetParameters=true, module=SetVariables_#6}
   &PartitionParticles(Number Of Partitions=16, Particles Per Partition=1, Partition Mode:ENiagara_PartitionMode_v2=::NewEnumerator1)
   &InitializeParticle(
    Color=&MergeHelper_ColorSwitch(
     0=&FNTimeOfDayColor(
      Night Color=<rgba 2.3125 1.1562 0 1>
      Morning Color=<rgba 2.3125 1.2719 0 1>
      Day Color=<rgba 2.3125 1.1562 0 1>
      Evening Color=<rgba 2.3125 0.9635 0 1>
     )
     1=&RandomRangeLinearColor(Minimum=<rgba 1 0.5281 0.0563 0.5>, Maximum=<rgba 1 0.45 0 0.5>)
     2=<rgba 3 1.8435 0.2244 1>
     3=<rgba 3 1.8438 0.225 1>
     Index=$Particles.VisibilityTag
     Num=::NewEnumerator2
    )
    Lifetime=&MergeHelper_FloatSwitch(
     0=&RandomRangeFloat(Minimum=0.75, Maximum=1.2, Randomness Mode=::NewEnumerator2, Random Seed=?)
     1=&RandomRangeFloat(Minimum=0.4, Maximum=1.25)
     2=0.1
     3=0.1
     Index=$Particles.VisibilityTag
     Num=::NewEnumerator2
    )
    Mesh Scale Mode=::NewEnumerator3
    Mesh Uniform Scale=0.225
    Position Offset=$System.BurstPosition
    Sprite Rotation Angle=-20
    Sprite Rotation Mode=::NewEnumerator0
    Sprite Size Mode=::NewEnumerator3
    Uniform Sprite Size=&Lerp_Float(
     A=10
     B=50
     Alpha=&FloatFromCurve(
      FloatCurve=DataInterface() {propertyValues {Curve {keys [
          {interpMode=RCIM_Cubic, tangentMode=RCTM_User, time=0, value=0.2}
          {arriveTangent=1.1314, interpMode=RCIM_Cubic, leaveTangent=1.1314, tangentMode=RCTM_User, time=0.9, value=0.6}
          {interpMode=RCIM_Cubic, tangentMode=RCTM_User, time=1, value=1}
         ]}}}
      CurveIndex=&RandomRangeFloat(Minimum=0, Maximum=1)
     )
    )
    UsePositionOffset=true
    Write Material Random=true
   )
   /SetVariables(
    Particles.GlitterSubUV:NiagaraFloat=&RandomRangeFloat(Minimum=0, Maximum=3)
    Particles.InitialRoundGlowSize:Vector2f=<100 100>
    Particles.InitialStarSize:Vector2f=&Multiply_Vector2D_ByFloat(Float=0.6, Vector2D=<30 40>)
   ) {isSetParameters=true, module=SetVariables_#7}
   &ShapeLocation(
    MaskSpawnGroup=false
    Offset=&MergeHelper_VectorSwitch(0=<0 0 0>, 1=<0 0 0>, 2=?, 3=?, Index=$Particles.VisibilityTag, Num=::NewEnumerator0)
    Offset Mode=::NewEnumerator0
    Sphere Distribution=::NewEnumerator3
    Sphere Radius=&MergeHelper_FloatSwitch(0=2, 1=1, 2=?, 3=?, Index=$Particles.VisibilityTag, Num=::NewEnumerator0)
    Uniform Particle Index:NiagaraInt32=&MergeHelper_IntSwitch(
     0=$Particles.PartitionParticles.Partition
     1=&ReturnExecIndex()
     2=0
     3=0
     4=?
     5=?
     Index=$Particles.VisibilityTag
     Num=::NewEnumerator2
    )
    Uniform Spiral Amount=0.5
    UseOwnerScale=true
   )
   &AddVelocity(
    Module.Use Falloff Distance=false
    Origin Offset=&MergeHelper_VectorSwitch(0=<0 0 -4>, 1=<0 0 -0.6>, 2=?, 3=?, Index=$Particles.VisibilityTag, Num=::NewEnumerator0)
    Override Seed=false
    Randomness Mode:ENiagaraRandomnessMode=::NewEnumerator2
    Velocity Mode=::NewEnumerator1
    Velocity Speed=&MergeHelper_FloatSwitch(
     0=0
     1=&RandomRangeFloat(Minimum=30, Maximum=300, Randomness Mode=::NewEnumerator2, Random Seed=?)
     2=?
     3=?
     Index=$Particles.VisibilityTag
     Num=::NewEnumerator0
    )
   )
   &AddVelocity(Module.Use Falloff Distance=false, Override Seed=false, Velocity Mode=::NewEnumerator1) {enabled=false, module=AddVelocity001}
   &InitialMeshOrientation(Mesh Axis To Orient=<0 1 0>, Mesh Orientation Mode=::NewEnumerator2, Use Rotation Vector=false) {enabled=false}
  ]
  particleUpdateScript [
   &ParticleState()
   &ScaleColor(
    Scale Alpha=&MergeHelper_FloatSwitch(
     0=1
     1=&Multiply_Float(
      A=0.25
      B=&FloatFromCurve(
       FloatCurve=DataInterface() {propertyValues {Curve {keys [<at 0 1>,<at 1 0>]}}}
       Scale Curve=&Waveform(
        Global Amplitude Scale=10
        "Waveform [1]"=::NewEnumerator0
        "[1] Phase Scale"=&RandomRangeFloat(Minimum=2, Maximum=4)
        "[1] Frequency"=&RandomRangeFloat(Minimum=0.75, Maximum=1)
        "[1] Amplitude Min/Max"=<0 1>
       )
      )
     )
     2=1
     3=1
     Index=$Particles.VisibilityTag
     Num=::NewEnumerator2
    )
    Scale RGB=&MergeHelper_VectorSwitch(
     0=&FN_Select_Vector_from_Time_Of_Day(Morning=<1.115 1.15 1.15>, Day=<1 1 1>, Evening=<1 1 1>, Night=<2.65 2.65 2.65>)
     1=<1 1 0>
     2=<1 1 1>
     3=<1 1 1>
     Index=0
     Num=::NewEnumerator2
    )
    ScaleA=true
    ScaleRGB=true
   )
   &GravityForce(Gravity=&Multiply_VectorByFloat(
     Vector=&MergeHelper_VectorSwitch(
      0=&RandomRangeVector(Minimum=<0 0 -80>, Maximum=<0 0 -110>, Randomness Mode=::NewEnumerator1, Random Seed=10, Fixed Random Seed=true)
      1=&RandomRangeVector(Minimum=<0 0 -250>, Maximum=<0 0 -50>)
      2=?
      3=?
      Index=$Particles.VisibilityTag
      Num=::NewEnumerator0
     )
     Float=&MergeHelper_FloatSwitch(
      0=&FloatFromCurve(
       FloatCurve=DataInterface() {propertyValues {Curve {keys [{arriveTangent=1.7842, interpMode=RCIM_Cubic, leaveTangent=1.7842, tangentMode=RCTM_User, time=0, value=1}, <at 1 2 RCIM_Cubic>]}}}
       Scale Curve=1.6
      )
      1=1
      2=?
      3=?
      Index=$Particles.VisibilityTag
      Num=::NewEnumerator0
     )
    ))
   &Drag(
    Drag=&MergeHelper_FloatSwitch(
     0=&FloatFromCurve(
      FloatCurve=DataInterface() {propertyValues {Curve {keys [<at 0 0 RCIM_Cubic>,<at 0.176 0.7545 RCIM_Cubic>,<at 0.3 0 RCIM_Cubic>]}}}
      Scale Curve=3
     )
     1=&RandomRangeFloat(Minimum=5, Maximum=12)
     2=?
     3=?
     Index=$Particles.VisibilityTag
     Num=::NewEnumerator0
    )
    Use Linear Drag=true
    Use Rotational Drag=true
   )
   &PointAttractionForce(
    Attraction Radius=15
    Attraction Strength=&MergeHelper_FloatSwitch(
     0=&FloatFromCurve(
      FloatCurve=DataInterface() {propertyValues {Curve {keys [
          {arriveTangent=36.0485, interpMode=RCIM_Cubic, leaveTangent=36.0485, tangentMode=RCTM_User, time=0, value=0}
          {arriveTangent=0.4397, interpMode=RCIM_Cubic, leaveTangent=0.4397, tangentMode=RCTM_User, time=0.089, value=1}
          {arriveTangent=-0.6185, interpMode=RCIM_Cubic, leaveTangent=-0.6185, tangentMode=RCTM_User, time=0.3, value=0}
         ]}}}
      Scale Curve=100
     )
     1=0
     2=0
     3=0
     Index=$Particles.VisibilityTag
     Num=::NewEnumerator2
    )
    Attractor Position Offset:Vector3f=&AddVectorToPosition(
     Position=&Multiply_Vector(
      A=&MaskPosition(
       Mask Source:NiagaraInt32=1
       Mask Parameter:NiagaraInt32=$Particles.PartitionParticles.Partition
       Mask:NiagaraInt32=0
       Masked Position:NiagaraPosition=<-48 0 17>
       Passthrough Position:NiagaraPosition=&MaskPosition(
        Mask Source:NiagaraInt32=1
        Mask Parameter:NiagaraInt32=$Particles.PartitionParticles.Partition
        Mask:NiagaraInt32=1
        Masked Position:NiagaraPosition=<-40 0 -29>
        Passthrough Position:NiagaraPosition=&MaskPosition(
         Mask Source:NiagaraInt32=1
         Mask Parameter:NiagaraInt32=$Particles.PartitionParticles.Partition
         Mask:NiagaraInt32=2
         Masked Position:NiagaraPosition=<40 0 -29>
         Passthrough Position:NiagaraPosition=&MaskPosition(
          Mask Source:NiagaraInt32=1
          Mask Parameter:NiagaraInt32=$Particles.PartitionParticles.Partition
          Mask:NiagaraInt32=3
          Masked Position:NiagaraPosition=<-25 0 -47>
          Passthrough Position:NiagaraPosition=&MaskPosition(
           Mask Source:NiagaraInt32=1
           Mask Parameter:NiagaraInt32=$Particles.PartitionParticles.Partition
           Mask:NiagaraInt32=4
           Masked Position:NiagaraPosition=<0 0 -92>
           Passthrough Position:NiagaraPosition=&MaskPosition(
            Mask Source:NiagaraInt32=1
            Mask Parameter:NiagaraInt32=$Particles.PartitionParticles.Partition
            Mask:NiagaraInt32=5
            Masked Position:NiagaraPosition=<14 0 10>
            Passthrough Position:NiagaraPosition=&MaskPosition(
             Mask Source:NiagaraInt32=1
             Mask Parameter:NiagaraInt32=$Particles.PartitionParticles.Partition
             Mask:NiagaraInt32=6
             Masked Position:NiagaraPosition=<0 0 -10>
             Passthrough Position:NiagaraPosition=&MaskPosition(
              Mask Source:NiagaraInt32=1
              Mask Parameter:NiagaraInt32=$Particles.PartitionParticles.Partition
              Mask:NiagaraInt32=7
              Masked Position:NiagaraPosition=<-12 0 10>
              Passthrough Position:NiagaraPosition=&MaskPosition(
               Mask Source:NiagaraInt32=1
               Mask Parameter:NiagaraInt32=$Particles.PartitionParticles.Partition
               Mask:NiagaraInt32=8
               Masked Position:NiagaraPosition=<48 0 17>
               Passthrough Position:NiagaraPosition=&MaskPosition(
                Mask Source:NiagaraInt32=1
                Mask Parameter:NiagaraInt32=$Particles.PartitionParticles.Partition
                Mask:NiagaraInt32=9
                Masked Position:NiagaraPosition=<28 0 28>
                Passthrough Position:NiagaraPosition=&MaskPosition(
                 Mask Source:NiagaraInt32=1
                 Mask Parameter:NiagaraInt32=$Particles.PartitionParticles.Partition
                 Mask:NiagaraInt32=10
                 Masked Position:NiagaraPosition=<-28 0 28>
                 Passthrough Position:NiagaraPosition=&MaskPosition(
                  Mask Source:NiagaraInt32=1
                  Mask Parameter:NiagaraInt32=$Particles.PartitionParticles.Partition
                  Mask:NiagaraInt32=11
                  Masked Position:NiagaraPosition=<50 0 -7>
                  Passthrough Position:NiagaraPosition=&MaskPosition(
                   Mask Source:NiagaraInt32=1
                   Mask Parameter:NiagaraInt32=$Particles.PartitionParticles.Partition
                   Mask:NiagaraInt32=12
                   Masked Position:NiagaraPosition=<12 0 -70>
                   Passthrough Position:NiagaraPosition=&MaskPosition(
                    Mask Source:NiagaraInt32=1
                    Mask Parameter:NiagaraInt32=$Particles.PartitionParticles.Partition
                    Mask:NiagaraInt32=13
                    Masked Position:NiagaraPosition=<-50 0 -7>
                    Passthrough Position:NiagaraPosition=&MaskVector(Mask Source=1, Mask Parameter=$Particles.PartitionParticles.Partition, Mask=14, Masked Vector=<-12 0 -70>, Passthrough Vector=<25 0 -47>)
                   )
                  )
                 )
                )
               )
              )
             )
            )
           )
          )
         )
        )
       )
      )
      B=&Multiply_VectorByFloat(Vector=<0.25 0.15 0.2>, Float=0.55)
     )
     Vector=&VectorFromCurve(
      VectorCurve=DataInterface() {propertyValues {XCurve {keys [<at 0 0>,<at 1 0>]}, YCurve {keys [<at 0 0>,<at 1 0>]}, ZCurve {keys [<at 0 1 RCIM_Cubic>,<at 0.7003 -1 RCIM_Cubic>]}}}
      Scale Curve=<1 1 5>
     )
    )
    Attractor Position Offset Coordinate Space:ENiagaraCoordinateSpace=::Local
    Falloff Exponent=1
    Kill Within Radius=false
    Module.Use Falloff=true
    UseAttractorOffset=true
   )
   /SetVariables(Particles.PositionHeart:NiagaraPosition=&Lerp_Vector(
     A=Linked() {linkedVariable {name=Particles.Position, type=NiagaraPosition}}
     B=&Multiply_Vector(
      A=&AddVectorToPosition(
       Position=&MaskPosition(
        Mask Source:NiagaraInt32=1
        Mask Parameter:NiagaraInt32=$Particles.PartitionParticles.Partition
        Mask:NiagaraInt32=0
        Masked Position:NiagaraPosition=<-48 0 17>
        Passthrough Position:NiagaraPosition=&MaskPosition(
         Mask Source:NiagaraInt32=1
         Mask Parameter:NiagaraInt32=$Particles.PartitionParticles.Partition
         Mask:NiagaraInt32=1
         Masked Position:NiagaraPosition=<-40 0 -29>
         Passthrough Position:NiagaraPosition=&MaskPosition(
          Mask Source:NiagaraInt32=1
          Mask Parameter:NiagaraInt32=$Particles.PartitionParticles.Partition
          Mask:NiagaraInt32=2
          Masked Position:NiagaraPosition=<40 0 -29>
          Passthrough Position:NiagaraPosition=&MaskPosition(
           Mask Source:NiagaraInt32=1
           Mask Parameter:NiagaraInt32=$Particles.PartitionParticles.Partition
           Mask:NiagaraInt32=3
           Masked Position:NiagaraPosition=<-25 0 -47>
           Passthrough Position:NiagaraPosition=&MaskPosition(
            Mask Source:NiagaraInt32=1
            Mask Parameter:NiagaraInt32=$Particles.PartitionParticles.Partition
            Mask:NiagaraInt32=4
            Masked Position:NiagaraPosition=<0 0 -92>
            Passthrough Position:NiagaraPosition=&MaskPosition(
             Mask Source:NiagaraInt32=1
             Mask Parameter:NiagaraInt32=$Particles.PartitionParticles.Partition
             Mask:NiagaraInt32=5
             Masked Position:NiagaraPosition=<14 0 10>
             Passthrough Position:NiagaraPosition=&MaskPosition(
              Mask Source:NiagaraInt32=1
              Mask Parameter:NiagaraInt32=$Particles.PartitionParticles.Partition
              Mask:NiagaraInt32=6
              Masked Position:NiagaraPosition=<0 0 -10>
              Passthrough Position:NiagaraPosition=&MaskPosition(
               Mask Source:NiagaraInt32=1
               Mask Parameter:NiagaraInt32=$Particles.PartitionParticles.Partition
               Mask:NiagaraInt32=7
               Masked Position:NiagaraPosition=<-12 0 10>
               Passthrough Position:NiagaraPosition=&MaskPosition(
                Mask Source:NiagaraInt32=1
                Mask Parameter:NiagaraInt32=$Particles.PartitionParticles.Partition
                Mask:NiagaraInt32=8
                Masked Position:NiagaraPosition=<48 0 17>
                Passthrough Position:NiagaraPosition=&MaskPosition(
                 Mask Source:NiagaraInt32=1
                 Mask Parameter:NiagaraInt32=$Particles.PartitionParticles.Partition
                 Mask:NiagaraInt32=9
                 Masked Position:NiagaraPosition=<28 0 28>
                 Passthrough Position:NiagaraPosition=&MaskPosition(
                  Mask Source:NiagaraInt32=1
                  Mask Parameter:NiagaraInt32=$Particles.PartitionParticles.Partition
                  Mask:NiagaraInt32=10
                  Masked Position:NiagaraPosition=<-28 0 28>
                  Passthrough Position:NiagaraPosition=&MaskPosition(
                   Mask Source:NiagaraInt32=1
                   Mask Parameter:NiagaraInt32=$Particles.PartitionParticles.Partition
                   Mask:NiagaraInt32=11
                   Masked Position:NiagaraPosition=<50 0 -7>
                   Passthrough Position:NiagaraPosition=&MaskPosition(
                    Mask Source:NiagaraInt32=1
                    Mask Parameter:NiagaraInt32=$Particles.PartitionParticles.Partition
                    Mask:NiagaraInt32=12
                    Masked Position:NiagaraPosition=<12 0 -70>
                    Passthrough Position:NiagaraPosition=&MaskPosition(
                     Mask Source:NiagaraInt32=1
                     Mask Parameter:NiagaraInt32=$Particles.PartitionParticles.Partition
                     Mask:NiagaraInt32=13
                     Masked Position:NiagaraPosition=<-50 0 -7>
                     Passthrough Position:NiagaraPosition=&MaskVector(Mask Source=1, Mask Parameter=$Particles.PartitionParticles.Partition, Mask=14, Masked Vector=<-12 0 -70>, Passthrough Vector=<25 0 -47>)
                    )
                   )
                  )
                 )
                )
               )
              )
             )
            )
           )
          )
         )
        )
       )
       Vector=<0 0 20>
      )
      B=&Multiply_VectorByFloat(Vector=<0.25 0.15 0.2>, Float=0.8)
     )
     Alpha=0.5
    )) {enabled=false, isSetParameters=true, module=SetVariables_#8}
   /SetVariables(Particles.PositionFreeze:NiagaraPosition=&Lerp_Vector(
     A=Linked() {linkedVariable {name=Particles.Position, type=NiagaraPosition}}
     B=Linked() {linkedVariable {name=Particles.PositionHeart, type=NiagaraPosition}}
     Alpha=1
    )) {enabled=false, isSetParameters=true, module=SetVariables_#9}
   &SolveForcesAndVelocity(Clamp Velocity=false, Limit Acceleration=false)
   &ScaleMeshSize(Scale Factor=&VectorFromFloat(Value=&FloatFromCurve(FloatCurve=DataInterface() {propertyValues {Curve {
         defaultValue=0
         keys [
          {interpMode=RCIM_Cubic, tangentMode=RCTM_User, time=0.3003, value=1}
          {arriveTangent=-7.9947, interpMode=RCIM_Cubic, leaveTangent=-7.9947, tangentMode=RCTM_User, time=0.6997, value=0}
         ]
        }}})))
   &ScaleColor(
    Scale RGB=&VectorFromFloat(Value=&MaskFloat(
      Mask Source=1
      Mask Parameter=$Particles.PartitionParticles.Partition
      Mask=0
      Masked Float=1
      Passthrough Float=&MaskFloat(
       Mask Source=1
       Mask Parameter=$Particles.PartitionParticles.Partition
       Mask=1
       Masked Float=1
       Passthrough Float=&MaskFloat(
        Mask Source=1
        Mask Parameter=$Particles.PartitionParticles.Partition
        Mask=2
        Masked Float=1
        Passthrough Float=&MaskFloat(
         Mask Source=1
         Mask Parameter=$Particles.PartitionParticles.Partition
         Mask=3
         Masked Float=1
         Passthrough Float=&MaskFloat(
          Mask Source=1
          Mask Parameter=$Particles.PartitionParticles.Partition
          Mask=4
          Masked Float=1
          Passthrough Float=&MaskFloat(
           Mask Source=1
           Mask Parameter=$Particles.PartitionParticles.Partition
           Mask=5
           Masked Float=1
           Passthrough Float=&MaskFloat(
            Mask Source=1
            Mask Parameter=$Particles.PartitionParticles.Partition
            Mask=6
            Masked Float=1
            Passthrough Float=&MaskFloat(
             Mask Source=1
             Mask Parameter=$Particles.PartitionParticles.Partition
             Mask=7
             Masked Float=1
             Passthrough Float=&MaskFloat(
              Mask Source=1
              Mask Parameter=$Particles.PartitionParticles.Partition
              Mask=8
              Masked Float=1
              Passthrough Float=&MaskFloat(
               Mask Source=1
               Mask Parameter=$Particles.PartitionParticles.Partition
               Mask=9
               Masked Float=1
               Passthrough Float=&MaskFloat(
                Mask Source=1
                Mask Parameter=$Particles.PartitionParticles.Partition
                Mask=10
                Masked Float=1
                Passthrough Float=&MaskFloat(
                 Mask Source=1
                 Mask Parameter=$Particles.PartitionParticles.Partition
                 Mask=11
                 Masked Float=1
                 Passthrough Float=&MaskFloat(
                  Mask Source=1
                  Mask Parameter=$Particles.PartitionParticles.Partition
                  Mask=12
                  Masked Float=1
                  Passthrough Float=&MaskFloat(
                   Mask Source=1
                   Mask Parameter=$Particles.PartitionParticles.Partition
                   Mask=13
                   Masked Float=1
                   Passthrough Float=&MaskFloat(Mask Source=1, Mask Parameter=$Particles.PartitionParticles.Partition, Mask=14, Masked Float=1, Passthrough Float=1)
                  )
                 )
                )
               )
              )
             )
            )
           )
          )
         )
        )
       )
      )
     ))
    ScaleA=true
    ScaleRGB=true
   ) {module=ScaleColor001}
   &UpdateMeshOrientation(
    Rotation Rate=&Multiply_Float(
     A=0
     B=&FloatFromCurve(
      FloatCurve=DataInterface() {propertyValues {Curve {
         defaultValue=1
         keys [
          {arriveTangent=6.9016, interpMode=RCIM_Cubic, leaveTangent=6.9016, tangentMode=RCTM_User, time=1.1108, value=0}
          {interpMode=RCIM_Cubic, tangentMode=RCTM_User, time=1.3484, value=1}
         ]
        }}}
      CurveIndex=$Emitter.Age
     )
    )
    Rotation Vector=<0 0 0.5>
   )
   &ScaleSpriteSize(Uniform Curve Sprite Scale=DataInterface() {propertyValues {Curve {keys [
        {interpMode=RCIM_Cubic, tangentMode=RCTM_User, time=0, value=0.2}
        {interpMode=RCIM_Cubic, tangentMode=RCTM_User, time=0.2, value=1}
        {tangentMode=RCTM_User, time=1, value=0}
       ]}}})
   &DynamicMaterialParameters(
    Index 0 Param 1=0.15
    Index 0 Param 2=&RandomRangeFloat(Minimum=0.25, Maximum=2)
    Index 0 Param 3=0.5
    Module.Index0Param0Write=true
    Module.Index0Param1Write=true
    Module.Index0Param2Write=true
    Module.Index0Param3Write=true
   )
   /SetVariables(
    Particles.Linear Color=&Multiply_LinearColorByFloat(Linear Color=$Particles.Color, Float=2)
    Particles.Vector 2D:Vector2f=&Multiply_Vector2D_ByFloat(Float=&RandomRangeFloat(Minimum=0.15, Maximum=0.25), Vector2D=$Particles.SpriteSize)
   ) {isSetParameters=true, module=SetVariables_#10}
   /SetVariables(
    Particles.UpdateRoundGlowSize:Vector2f=&Multiply_Vector2D(
     A=$Particles.InitialRoundGlowSize
     B=&Vector2DFromFloat(Value=&FloatFromCurve(FloatCurve=DataInterface() {propertyValues {Curve {
          defaultValue=1
          keys [
           {arriveTangent=4.7075, interpMode=RCIM_Cubic, leaveTangent=4.7075, tangentMode=RCTM_User, time=0, value=0}
           {arriveTangent=-1.2993, interpMode=RCIM_Cubic, leaveTangent=-1.2993, tangentMode=RCTM_User, time=0.4887, value=0.9847}
           {arriveTangent=-0.4432, interpMode=RCIM_Cubic, leaveTangent=-0.4432, tangentMode=RCTM_User, time=1, value=0}
          ]
         }}}))
    )
    Particles.UpdateStarSize:Vector2f=&Multiply_Vector2D(
     A=$Particles.InitialStarSize
     B=&Vector2DFromCurve(
      Vector2Curve=DataInterface() {propertyValues {
        XCurve {
         defaultValue=0
         keys [
          {arriveTangent=0.0865, interpMode=RCIM_Cubic, leaveTangent=0.0865, tangentMode=RCTM_User, time=0, value=0}
          {arriveTangent=4.9825, interpMode=RCIM_Cubic, leaveTangent=4.9825, tangentMode=RCTM_User, time=0.4724, value=1}
          {arriveTangent=-1.9029, interpMode=RCIM_Cubic, leaveTangent=-1.9029, time=0.8093, value=1.9173}
          <at 0.9979 0 RCIM_Cubic>
         ]
        }
        YCurve {
         defaultValue=1
         keys [
          {arriveTangent=-5.5781, interpMode=RCIM_Cubic, leaveTangent=-5.5781, tangentMode=RCTM_User, time=0.6, value=1.1}
          {arriveTangent=-0.1643, interpMode=RCIM_Cubic, leaveTangent=-0.1643, tangentMode=RCTM_User, time=1, value=0}
         ]
        }
       }}
      ScaleCurve=&RandomRangeVector2D(Minimum=<1 1>, Maximum=<1 1.2>)
     )
    )
   ) {isSetParameters=true, module=SetVariables_#11}
   /SetVariables(
    Particles.CameraOffsetStar=15
    Particles.GlowColor=&MakeLinearColorFromVectorAndFloat(
     "Vector (RGB)"=&Multiply_Vector(A=&MakeVectorFromLinearColorRGB(Color=<rgba 1.2 0.6795 0 1>), B=<1 1 1>)
     "Float (Alpha)"=1
    )
    Particles.GlowSize=&Multiply_Vector2D_ByFloat(Float=1.75, Vector2D=$Particles.UpdateStarSize)
    Particles.RoundGlow:LinearColor=&MakeLinearColorFromVectorAndFloat(
     "Vector (RGB)"=&Multiply_Vector(A=&MakeVectorFromLinearColorRGB(Color=<rgba 1.2 0.6795 0 1>), B=<1 1 1>)
     "Float (Alpha)"=0.15
    )
   ) {isSetParameters=true, module=SetVariables_#12}
  ]
  rendererClasses [&NiagaraMeshRendererProperties, &NiagaraSpriteRendererProperties]
  renderers [
   {rendererClass=&NiagaraMeshRendererProperties, rendererIndex=0}
   {rendererClass=&NiagaraMeshRendererProperties, rendererIndex=1}
   {rendererClass=&NiagaraSpriteRendererProperties, rendererIndex=2}
   {rendererClass=&NiagaraSpriteRendererProperties, rendererIndex=3}
   {rendererClass=&NiagaraSpriteRendererProperties, rendererIndex=4}
   {rendererClass=&NiagaraSpriteRendererProperties, rendererIndex=5}
   {rendererClass=&NiagaraSpriteRendererProperties, rendererIndex=6}
  ]
 }
}
summary {
 systemName=NS_ReactFX_SizeTribe
 emitters [
  {emitterName=Intro, rendererClasses [&NiagaraMeshRendererProperties]}
  {
   bEnabled=false
   emitterName=Pop_V1
   rendererClasses [&NiagaraMeshRendererProperties, &NiagaraSpriteRendererProperties]
  }
  {emitterName=Pop_FaceCamera, rendererClasses [&NiagaraMeshRendererProperties, &NiagaraSpriteRendererProperties]}
  {bEnabled=false, emitterName=Intro001, rendererClasses [&NiagaraMeshRendererProperties]}
  {
   bEnabled=false
   emitterName=Pop_FaceCamera001
   rendererClasses [&NiagaraMeshRendererProperties, &NiagaraSpriteRendererProperties]
  }
 ]
 userVariables [
  User.bIsFrontEnd {type=NiagaraBool}
  User.bIsFrontEndPreview {type=NiagaraBool}
 ]
}
