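# Niagara reaction effect for cosmetic companions (SniffRust variant): a letter swooshes in with petting audio, then gives way to a heart.
# System.IntroLength sequences the swoosh mesh, letter sprite and delayed heart, all anchored at System.BurstPosition.
# One confetti emitter fires letter and heart-explosion bursts, split by spawn group, projecting a hardcoded point array onto camera axes for the heart shape.
# Front-end values adjust sizes and offsets.
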NiagaraSystem /CosmeticCompanions/FX/React/NS_ReactFX_SniffRust
dependencies {
 usedDataInterfaces [
  &NiagaraDataInterfaceAudioPlayer
  &NiagaraDataInterfaceCurve
  &NiagaraDataInterfaceArrayFloat3
  &NiagaraDataInterfaceCamera
  &NiagaraDataInterfaceVector2DCurve
  &NiagaraDataInterfaceColorCurve
 ]
 usedDynamicInputs [
  &SimulationPosition
  &FN_FrontEndVector3Modifier
  &ColorFromCurve
 ]
 usedModules [
  &self:SystemSpawnScript.SystemScriptSource.SystemScriptGraph.NiagaraNodeAssignment_0.SetVariables_#1
  &self:SystemSpawnScript.SystemScriptSource.SystemScriptGraph.NiagaraNodeAssignment_6.SetVariables_#2
  &self:SystemSpawnScript.SystemScriptSource.SystemScriptGraph.NiagaraNodeAssignment_5.SetVariables_#3
  &self:SystemSpawnScript.SystemScriptSource.SystemScriptGraph.NiagaraNodeAssignment_2.SetVariables_#4
  &self:SystemSpawnScript.SystemScriptSource.SystemScriptGraph.NiagaraNodeAssignment_4.SetVariables_#5
  &SystemState
 ]
}
emitters {
 Heart {
  emitterUpdateScript [
   &EmitterState(
    Life Cycle Mode=::NewEnumerator1
    Loop Behavior=::NewEnumerator1
    Loop Delay=&Add_Float(A=$System.IntroLength, B=0.22)
    Loop Duration=1
    UseLoopDelay=true
   )
   &SpawnBurst_Instantaneous(Use Spawn Probability=false, UseLoopCountLimit=false)
  ]
  particleSpawnScript [
   /SetVariables(Particles.VisibilityTag=$Emitter.SpawnGroup) {isSetParameters=true, module=SetVariables_#6}
   &InitializeParticle(
    Color=&LerpLinearColors(StartColor=<rgba 1 1 1 1>, EndColor=$System.TOD, LerpFactor=0.8)
    Lifetime=$User.HeartLength
    Sprite Rotation Mode=::NewEnumerator0
    Sprite Size Mode=::NewEnumerator3
    Uniform Sprite Size=&Multiply_Float(A=23, B=&FN_FrontEndFloatModifier(Gameplay=1, Front End=$User.FrontEndScale))
    UsePositionOffset=false
    Write Material Random=true
   )
  ]
  particleUpdateScript [
   &ParticleState()
   &ScaleSpriteSize(
    Non-Uniform Curve Sprite Scale=DataInterface() {propertyValues {
      XCurve {keys [<at 0 0.2 RCIM_Cubic>, {interpMode=RCIM_Cubic, leaveTangent=0.625, time=0.2, value=1}]}
      YCurve {keys [<at 0 2.5 RCIM_Cubic>,<at 0.2 1 RCIM_Cubic>]}
     }}
    Scale Sprite Size Mode=::NewEnumerator3
   )
   &ScaleSpriteSize(
    Scale Sprite Size Mode=::NewEnumerator0
    Uniform Scale Factor=&Add_Float(
     A=0
     B=&FloatFromCurve(FloatCurve=DataInterface() {propertyValues {Curve {
         defaultValue=0
         keys [
          <at 0.1 1>
          {arriveTangent=1.0432, interpMode=RCIM_Cubic, leaveTangent=1.0432, tangentMode=RCTM_User, time=0.75, value=1.05}
          {arriveTangent=3.4731, interpMode=RCIM_Cubic, leaveTangent=3.4731, tangentMode=RCTM_User, time=0.9, value=1.2}
          {arriveTangent=17.4696, interpMode=RCIM_Cubic, leaveTangent=17.4696, tangentMode=RCTM_User, time=1, value=3}
         ]
        }}})
    )
   ) {module=ScaleSpriteSize002}
   &ScaleColor(
    Scale Alpha=&FloatFromCurve(FloatCurve=DataInterface() {propertyValues {Curve {
        defaultValue=0
        keys [
         {arriveTangent=9.0725, interpMode=RCIM_Cubic, leaveTangent=9.0725, tangentMode=RCTM_User, time=0, value=0}
         {arriveTangent=-0.2441, leaveTangent=-0.2441, time=0.1, value=0.9847}
         {arriveTangent=-1.146, leaveTangent=-1.146, time=0.98, value=1}
        ]
       }}})
    ScaleA=true
    ScaleRGB=true
   )
   /SetVariables(Particles.Position=&AddVectorToPosition(
     Position=Linked() {linkedVariable {name=System.BurstPosition, type=Vector3f}}
     Vector=&Multiply_VectorByFloat(Vector=&ReturnCameraVector(Camera Vectors=::NewEnumerator2, Coordinate Space=::Simulation), Float=2)
    )) {isSetParameters=true, module=SetVariables_#7}
  ]
  rendererClasses [&NiagaraSpriteRendererProperties]
  renderers [{rendererClass=&NiagaraSpriteRendererProperties}]
 }
 HeartRadialGlow {rendererClasses [], renderers []}
 HeartShapedGlow {rendererClasses [], renderers []}
 LWE_AirGlow {rendererClasses [], renderers []}
 Letter_2 {
  emitterUpdateScript [
   &EmitterState(
    Life Cycle Mode=::NewEnumerator1
    Loop Behavior=::NewEnumerator1
    Loop Delay=&Subtract_Float(A=$System.IntroLength, B=0)
    Loop Duration=1
    UseLoopDelay=true
   )
   &SpawnBurst_Instantaneous(Use Spawn Probability=false, UseLoopCountLimit=false)
   /SetVariables(Emitter.DistortionAmount:NiagaraFloat=&FloatFromCurve(
     FloatCurve=DataInterface() {propertyValues {Curve {keys [<at 0 0 RCIM_Cubic>,<at 0.45 1 RCIM_Cubic>]}}}
     CurveIndex=&Divide_Float(A=$Emitter.NormalizedLoopAge, B=0.5)
     Scale Curve=0.75
    )) {isSetParameters=true}
  ]
  particleSpawnScript [&InitializeParticle(
    Color=$System.TOD
    Lifetime=0.25
    Sprite Rotation Mode=::NewEnumerator0
    Sprite Size=&Multiply_Vector2D_ByFloat(
     Float=&Multiply_Float(A=1.1, B=$System.FE_Scale)
     Vector2D=&Multiply_Vector2D_ByFloat(Float=1.5, Vector2D=<26 21>)
    )
    Sprite Size Mode=::NewEnumerator4
    UsePositionOffset=false
    Write Material Random=true
   )]
  particleUpdateScript [
   &ParticleState()
   &ScaleSpriteSize(
    Scale Factor=&Vector2DFromCurve(
     Vector2Curve=DataInterface() {propertyValues {
       XCurve {keys [
         {arriveTangent=-0.0537, interpMode=RCIM_Cubic, leaveTangent=-0.0537, tangentMode=RCTM_User, time=0.75, value=1}
         {arriveTangent=-2, interpMode=RCIM_Cubic, leaveTangent=-2, time=0.95, value=1}
         <at 1 0.5 RCIM_Cubic>
        ]}
       YCurve {keys [{arriveTangent=6.5059, interpMode=RCIM_Cubic, leaveTangent=6.5059, tangentMode=RCTM_User, time=0.75, value=1}, <at 1 2 RCIM_Cubic>]}
      }}
     ScaleCurve=<1 1>
    )
    Scale Sprite Size Mode=::NewEnumerator2
   ) {module=ScaleSpriteSize002}
   &ScaleColor(
    Scale Alpha=&FloatFromCurve(FloatCurve=DataInterface() {propertyValues {Curve {defaultValue=0, keys [{arriveTangent=-0.2441, time=0.05, value=0.9847}, {arriveTangent=-1.146, leaveTangent=-1.146, time=0.98, value=1}]}}})
    Scale RGB=&MakeVectorFromLinearColorRGB(Color=<rgba 1 1 1 1>)
    ScaleA=true
    ScaleRGB=true
   )
   /SetVariables(Particles.Position=&AddVectorToPosition(
     Position=Linked() {linkedVariable {name=System.BurstPosition, type=Vector3f}}
     Vector=&Multiply_VectorByFloat(
      Vector=&ReturnCameraVector(Camera Vectors=::NewEnumerator2, Coordinate Space=::Simulation)
      Float=&FN_FrontEndFloatModifier(Gameplay=3, Front End=1)
     )
    )) {isSetParameters=true, module=SetVariables_#7}
  ]
  rendererClasses [&NiagaraSpriteRendererProperties]
  renderers [{rendererClass=&NiagaraSpriteRendererProperties}]
 }
 NE_Confetti {
  emitterUpdateScript [
   &EmitterState()
   &SpawnBurst_Instantaneous(
    Spawn Count=$User.LetterConfettiAmount
    Spawn Time=&Add_Float(A=$System.IntroLength, B=0.23)
    Use Spawn Probability=false
   )
   &SpawnBurst_Instantaneous(
    Spawn Count=58
    Spawn Group=1
    Spawn Time=&Add_Float(A=$System.HeartExplosionTiming, B=0)
    Use Spawn Probability=false
   ) {module=SpawnBurst_Instantaneous001}
  ]
  particleSpawnScript [
   /SetVariables(Particles.VisibilityTag=$Emitter.SpawnGroup) {isSetParameters=true}
   /SetVariables(Particles.CustomID:NiagaraInt32=&MakeIntFromFloat(Float=&ClampFloat(
      Float=&Divide_Float(
       A=&MakeFloatFromInt(Int=/Niagara/DynamicInputs/Subtract/Subtract_Integer(
         A:NiagaraInt32=&ReturnParticleIDAttributes()
         B:NiagaraInt32=&MakeIntFromFloat(Float=&Multiply_Float(A=&MakeFloatFromInt(Int=$User.LetterConfettiAmount), B=$Engine.Emitter.SpawnCountScale))
        ))
       B=$Engine.Emitter.SpawnCountScale
      )
      Max=100
     ))) {isSetParameters=true}
   /SetVariables(Particles.CustomVisTag:NiagaraInt32=&MakeCustomIntFromBool(
     Bool=&SetBoolByIntComparison(A=$Particles.VisibilityTag, B=0, Comparison Type=::NewEnumerator0)
     True INT=&MakeCustomIntFromBool(
      Bool=&SetBoolByIntComparison(A=$Particles.CustomID, B=27, Comparison Type=::NewEnumerator0)
      True INT=1
      False INT=2
     )
     False INT=0
    )) {isSetParameters=true}
   &InitializeParticle(
    Lifetime=&MergeHelper_FloatSwitch(
     0=&RandomRangeFloat(Minimum=0.2, Maximum=0.5)
     1=&FloatFromCurve(
      FloatCurve=DataInterface() {propertyValues {Curve {keys [<at 0 0.2>,<at 0.8 1>,<at 1 2>]}}}
      CurveIndex=&RandomRangeFloat(Minimum=0, Maximum=1)
     )
     2=?
     3=?
     Index=$Particles.VisibilityTag
     Num=::NewEnumerator0
    )
    Position Offset Coordinate Space=::Local
    Sprite Rotation Mode=::NewEnumerator2
    Sprite Size Mode=::NewEnumerator3
    Uniform Sprite Size=&Multiply_Float(A=&RandomRangeFloat(Minimum=2, Maximum=6), B=$System.FE_Scale)
    UsePositionOffset=true
    Write Material Random=true
   ) {module=InitializeParticle001}
   &ShapeLocation(
    MaskSpawnGroup=false
    Offset Mode=::NewEnumerator0
    Override Seed=false
    Sphere Radius=&MergeHelper_FloatSwitch(0=15, 1=15, 2=?, 3=?, Index=$Particles.VisibilityTag, Num=::NewEnumerator0)
    UseOwnerScale=true
   ) {enabled=false}
   /SetVariables(Particles.HeartPoint:Vector3f=&SelectVectorFromArray(
     Direct Array Index=&ModuloInt(A=$Particles.CustomID, B=50)
     Vector Selection Array=DataInterface() {propertyValues {FloatData [
        <0.001 -0.5952 0.8294>
        <0.001 -0.3021 0.7634>
        <-0.0109 -0.1539 -1.1263>
        <-0.0107 -0.1559 -1.1176>
        <0.0048 -0.6281 -0.6506>
        <0.001 0.336 0.7876>
        <0.0009 0.4208 0.7763>
        <-0.0111 0.1261 -1.2116>
        <-0.005 0.1913 -1.0214>
        <0.0041 0.7966 -0.488>
        <0.0034 0.7489 -0.4578>
        <0.0011 1.1331 0.4432>
        <0.0011 1.0351 0.5769>
        <0.001 0.8807 0.8471>
        <0.0019 -1.14 0.0228>
        <0.0019 -1.0875 0.0029>
        <0.0035 0.4651 -0.7579>
        <0.0011 0.6893 0.9042>
        <0.0028 1.1403 -0.1591>
        <0.001 0.1829 0.6756>
        <0.0008 0.0847 0.5191>
        <0.0011 -0.6428 0.9356>
        <0.0037 -0.9254 -0.3829>
        <0.001 -0.9484 0.7688>
        <0.0011 -1.1265 0.5026>
        <0.0011 -1.191 0.3279>
        <0.0009 -0.223 0.6893>
        <0.0009 -0.5103 0.5889>
        <-0.0009 -0.2908 0.0058>
        <0.0006 -0.1345 0.2868>
        <0.0035 -0.6141 -0.508>
        <-0.0145 -0.0238 -1.1077>
        <0.001 -0.9972 0.0434>
        <0.0009 -0.8381 0.23>
        <0.0009 -0.8263 0.3827>
        <0.0002 -0.3724 -0.4419>
        <0.0006 -0.5804 0.1621>
        <0.0008 -0.1903 0.4461>
        <0.002 -0.7872 -0.2237>
        <-0.0041 -0.0698 -0.2032>
        <0 -0.3212 0.1678>
        <0.0003 -0.4654 0.1852>
        <0.0005 -0.3909 0.3264>
        <0.0008 -0.3471 0.4871>
        <-0.0009 -0.1761 -0.8396>
        <0.0009 0.613 0.506>
        <-0.0009 0.0266 0.1218>
        <-0.0072 0.1101 -0.9713>
        <0.0009 0.85 0.2009>
        <-0.001 0.2064 -0.6464>
        <-0.0022 0.1698 -0.5728>
        <-0.0021 0.2233 -0.1943>
        <0.0006 0.5878 -0.0946>
        <0.001 0.6557 -0.1195>
        <0.0003 0.4865 0.1783>
        <0.0022 0.7253 -0.2757>
        <0.0033 0.4025 -0.7057>
        <-0.003 0.1276 -0.6837>
       ]}}
    )) {isSetParameters=true}
   /SetVariables(
    Particles.LetterPoint:Vector3f=&Multiply_VectorByFloat(
     Vector=&Multiply_Vector(A=&RandomRangeVector(Minimum=<0 -1 -1>, Maximum=<0 1 1>), B=<0 7 25>)
     Float=$System.FE_Scale
    )
    Particles.RotatedHeartPoint:Vector3f=&Multiply_VectorByFloat(
     Vector=&Add_Vector(
      A=&Multiply_VectorByFloat(
       Vector=&ReturnCameraVector(Camera Vectors=::NewEnumerator3, Coordinate Space=::Simulation)
       Float=&MakeFloatFromVector(
        Vector=&Multiply_VectorByFloat(
         Vector=$Particles.HeartPoint
         Float=&MergeHelper_FloatSwitch(0=0.8, 1=0.5, 2=0.7, 3=?, Index=$Particles.CustomVisTag, Num=::NewEnumerator1)
        )
        Channel=::NewEnumerator1
       )
      )
      B=&Multiply_VectorByFloat(
       Vector=&ReturnCameraVector(Camera Vectors=::NewEnumerator2, Coordinate Space=::Simulation)
       Float=&Add_Float(
        A=&MakeFloatFromVector(
         Vector=&Multiply_VectorByFloat(
          Vector=$Particles.HeartPoint
          Float=&MergeHelper_FloatSwitch(0=1, 1=0.5, 2=0.7, 3=?, Index=$Particles.CustomVisTag, Num=::NewEnumerator1)
         )
         Channel=::NewEnumerator2
        )
        B=0.2
       )
      )
     )
     Float=1
    )
   ) {isSetParameters=true}
   /SetVariables(Particles.Position=&AddVectorToPosition(
     Position=Linked() {linkedVariable {name=System.BurstPosition, type=Vector3f}}
     Vector=&MergeHelper_VectorSwitch(
      0=&Multiply_Vector(
       A=&Add_Vector(
        A=&Multiply_VectorByFloat(
         Vector=&ReturnCameraVector(Camera Vectors=::NewEnumerator3, Coordinate Space=::Simulation)
         Float=&MakeFloatFromVector(Vector=$Particles.LetterPoint, Channel=::NewEnumerator1)
        )
        B=&Multiply_VectorByFloat(
         Vector=&ReturnCameraVector(Camera Vectors=::NewEnumerator2, Coordinate Space=::Simulation)
         Float=&MakeFloatFromVector(Vector=$Particles.LetterPoint, Channel=::NewEnumerator2)
        )
       )
       B=<1 1 1>
      )
      1=&Add_Vector(A=&Multiply_VectorByFloat(Vector=$Particles.RotatedHeartPoint, Float=15), B=<0 0 0>)
      2=?
      3=?
      Index=$Particles.VisibilityTag
      Num=::NewEnumerator0
     )
    )) {isSetParameters=true}
   &AddVelocity(
    Velocity=&Add_Vector(
     A=&MergeHelper_VectorSwitch(
      0=&Multiply_VectorByFloat(
       Vector=&Subtract_Position(A=$Particles.Position, B=Linked() {linkedVariable {name=System.BurstPosition, type=Vector3f}})
       Float=0.05
      )
      1=&Multiply_VectorByFloat(Vector=$Particles.RotatedHeartPoint, Float=0.8)
      2=?
      3=?
      Index=$Particles.VisibilityTag
      Num=::NewEnumerator0
     )
     B=&Add_Vector(
      A=&Multiply_VectorByFloat(
       Vector=&ReturnCameraVector(Camera Vectors=::NewEnumerator1, Coordinate Space=::Simulation)
       Float=&RandomRangeFloat(Minimum=0, Maximum=0)
      )
      B=&Multiply_VectorByFloat(
       Vector=&ReturnCameraVector(Camera Vectors=::NewEnumerator2, Coordinate Space=::Simulation)
       Float=&RandomRangeFloat(Minimum=0.2, Maximum=0.5)
      )
     )
    )
    Velocity Speed Scale=&Multiply_Float(
     A=&MergeHelper_FloatSwitch(
      0=&RandomRangeFloat(Minimum=150, Maximum=200)
      1=&RandomRangeFloat(Minimum=200, Maximum=250)
      2=?
      3=?
      Index=$Particles.VisibilityTag
      Num=::NewEnumerator0
     )
     B=$System.FE_Scale
    )
   ) {module=AddVelocity001}
  ]
  particleUpdateScript [
   &ParticleState()
   &SpriteRotationRate(Rotation Rate=&RandomRangeFloat(Minimum=-50, Maximum=50))
   &Drag(Drag=&RandomRangeFloat(Minimum=5, Maximum=8), Use Linear Drag=true, Use Rotational Drag=false)
   &GravityForce(Gravity=&RandomRangeVector(Minimum=<-50 -50 -500>, Maximum=<50 50 -200>))
   &SolveForcesAndVelocity(Clamp Velocity=false, Limit Acceleration=false)
   &ScaleSpriteSize(Scale Factor=&Vector2DFromFloat(Value=&FloatFromCurve(FloatCurve=DataInterface() {propertyValues {Curve {keys [
          {arriveTangent=-0.4596, interpMode=RCIM_Cubic, leaveTangent=-0.4596, tangentMode=RCTM_User, time=0.55, value=1}
          {arriveTangent=-4.9031, interpMode=RCIM_Cubic, leaveTangent=-4.9031, tangentMode=RCTM_User, time=1, value=0}
         ]}}})))
   &ScaleColor(ScaleA=false, ScaleRGB=true)
  ]
  rendererClasses [&NiagaraSpriteRendererProperties]
  renderers [
   {rendererClass=&NiagaraSpriteRendererProperties}
   {rendererClass=&NiagaraSpriteRendererProperties, rendererIndex=1}
   {rendererClass=&NiagaraSpriteRendererProperties, rendererIndex=2}
  ]
 }
 SwooshLetter {
  emitterSpawnScript [&InitializeAudioPlayer(Audio Player=DataInterface() {propertyValues {
      Attenuation=null
      Concurrency=null
      MaxPlaysPerTick=10
      SoundToPlay=/CosmeticCompanions/Assets/Audio_Shared/Interaction/Petting/SC_Mimosa_Interaction_React_Default
      bAllowLoopingOneShotSounds=false
      bLimitPlaysPerTick=true
      bOnlyActiveDuringGameplay=false
      bStopWhenComponentIsDestroyed=true
      ConfigurationUserParameter {parameter {name=None, typeDefHandle {registeredTypeIndex=1}, varData []}}
      ParameterNames []
     }})]
  emitterUpdateScript [
   &EmitterState(
    Life Cycle Mode=::NewEnumerator1
    Loop Behavior=::NewEnumerator1
    Loop Duration=&Multiply_Float(A=$System.IntroLength, B=0.75)
    UseLoopDelay=false
   )
   /Niagara/Modules/Audio/PlayAudio(Play Audio Trigger:NiagaraBool=&SetBoolByFloatComparison(A=$Emitter.Age, B=0.5)) {module=PlayAudio001}
   &SpawnBurst_Instantaneous(Use Spawn Probability=false, UseLoopCountLimit=false)
   /SetVariables(
    Emitter.DistortionStrength:NiagaraFloat=&Lerp_Float(
     A=0.75
     B=0
     Alpha=&FloatFromCurve(
      FloatCurve=DataInterface() {propertyValues {Curve {
         defaultValue=1
         keys [<at 0.3 0 RCIM_Cubic>, {arriveTangent=2.9113, interpMode=RCIM_Cubic, leaveTangent=2.9113, tangentMode=RCTM_User, time=0.7, value=1}]
        }}}
      CurveIndex=&Divide_Float(A=$Emitter.Age, B=$Emitter.CurrentLoopDuration)
     )
    )
    Emitter.Offset:NiagaraFloat=&Lerp_Float(
     A=-30
     B=9.2623
     Alpha=&FloatFromCurve(
      FloatCurve=DataInterface() {propertyValues {Curve {
         defaultValue=1
         keys [
          {arriveTangent=3.2823, interpMode=RCIM_Cubic, leaveTangent=3.2823, tangentMode=RCTM_User, time=0, value=0}
          {arriveTangent=1.275, interpMode=RCIM_Cubic, leaveTangent=1.275, time=0.2615, value=0.7}
          {arriveTangent=0.0555, interpMode=RCIM_Cubic, leaveTangent=0.0555, tangentMode=RCTM_User, time=0.8, value=1.02}
          {arriveTangent=0.2993, interpMode=RCIM_Cubic, leaveTangent=0.2993, tangentMode=RCTM_User, time=1, value=1}
         ]
        }}}
      CurveIndex=&Divide_Float(A=$Emitter.Age, B=$Emitter.CurrentLoopDuration)
     )
    )
    Emitter.Tiling:NiagaraFloat=&Lerp_Float(
     A=0.1
     B=1
     Alpha=&FloatFromCurve(
      FloatCurve=DataInterface() {propertyValues {Curve {
         defaultValue=1
         keys [{arriveTangent=2.3592, interpMode=RCIM_Cubic, time=0.1, value=0}, {interpMode=RCIM_Cubic, leaveTangent=2.459, time=1, value=1}]
        }}}
      CurveIndex=&Divide_Float(A=$Emitter.Age, B=$Emitter.CurrentLoopDuration)
     )
    )
   ) {isSetParameters=true}
  ]
  particleSpawnScript [
   /SetVariables(Particles.VisibilityTag=$Emitter.SpawnGroup) {isSetParameters=true, module=SetVariables_#6}
   &InitializeParticle(
    Color=$System.TOD
    Lifetime=&Add_Float(A=$System.IntroLength, B=0.05)
    Mesh Scale Mode=::NewEnumerator3
    Mesh Uniform Scale=&Multiply_Float(A=1.5, B=$System.FE_Scale)
    Position Offset=<0 0 -5>
    Sprite Rotation Mode=::NewEnumerator0
    Sprite Size Mode=::NewEnumerator3
    Uniform Sprite Size=1
    UsePositionOffset=true
    Write Material Random=true
   )
  ]
  particleUpdateScript [
   &ParticleState()
   &UpdateMeshOrientation(
    Rotation Rate=&FloatFromCurve(
     FloatCurve=DataInterface() {propertyValues {Curve {keys [<at 0 0 RCIM_Cubic>,<at 0.3 1 RCIM_Cubic>]}}}
     CurveIndex=&Subtract_Float(A=$Particles.Age, B=$Emitter.CurrentLoopDuration)
     Scale Curve=0.05
    )
    Rotation Vector=&RandomRangeVector(Minimum=<-1 -1 -1>, Maximum=<1 1 1>)
   )
   &CurlNoiseForce(Bias Noise=false, Noise Frequency=5000, Noise Strength=1.5, Pan Noise=false)
   &SolveForcesAndVelocity(Clamp Velocity=false, Limit Acceleration=false)
   /SetVariables(Particles.Position=&AddVectorToPosition(
     Position=Linked() {linkedVariable {name=System.BurstPosition, type=Vector3f}}
     Vector=&Multiply_VectorByFloat(
      Vector=$Output.CurlNoiseForce.CurlNoiseForce
      Float=&FloatFromCurve(
       FloatCurve=DataInterface() {propertyValues {Curve {keys [
           {arriveTangent=0.7188, interpMode=RCIM_Cubic, leaveTangent=0.7188, tangentMode=RCTM_User, time=0, value=0.2}
           {interpMode=RCIM_Cubic, leaveTangent=0.7145, time=0.5, value=1}
          ]}}}
       CurveIndex=&Subtract_Float(A=$Particles.Age, B=$Emitter.CurrentLoopDuration)
      )
     )
    )) {isSetParameters=true}
  ]
  rendererClasses [&NiagaraMeshRendererProperties]
  renderers [{rendererClass=&NiagaraMeshRendererProperties, rendererIndex=0}]
 }
}
summary {
 systemName=NS_ReactFX_SniffRust
 emitters [
  {emitterName=LWE_AirGlow, rendererClasses []}
  {emitterName=SwooshLetter, rendererClasses [&NiagaraMeshRendererProperties]}
  {emitterName=NE_Confetti, rendererClasses [&NiagaraSpriteRendererProperties]}
  {emitterName=Heart, rendererClasses [&NiagaraSpriteRendererProperties]}
  {emitterName=HeartRadialGlow, rendererClasses []}
  {emitterName=Letter_2, rendererClasses [&NiagaraSpriteRendererProperties]}
  {emitterName=HeartShapedGlow, rendererClasses []}
 ]
 userVariables [
  User.bIsFrontEnd {type=NiagaraBool}
  User.bIsFrontEndPreview {type=NiagaraBool}
  User.FrontEndScale {type=NiagaraFloat, defaultValue=0.6, description=""}
  User.HeartLength {type=NiagaraFloat, defaultValue=0.7, description=""}
  User.LetterConfettiAmount {type=NiagaraInt32, defaultValue=20}
 ]
}
