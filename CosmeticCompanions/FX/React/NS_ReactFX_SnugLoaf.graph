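# Niagara reaction effect for the SnugLoaf cosmetic companion: a coffee-themed pour and splash ending in a burst, with a petting reaction sound.
# Intro plays the sound and a scaling sprite, while short-lived emitters draw a steam/coffee ribbon and gravity-driven droplets around shared system positions.
# A one-second-delayed burst multiplexes smoke, streaks, droplets, shards and handle through one emitter, using spawn groups as visibility tags.
# Front-end modifiers retune sizes and motion outside gameplay.
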NiagaraSystem /CosmeticCompanions/FX/React/NS_ReactFX_SnugLoaf
dependencies {
 usedDataInterfaces [
  &NiagaraDataInterfaceAudioPlayer
  &NiagaraDataInterfaceSpriteRendererInfo
  &NiagaraDataInterfaceCurve
  &NiagaraDataInterfaceArrayColor
  &NiagaraDataInterfaceCamera
  &NiagaraDataInterfaceColorCurve
  &NiagaraDataInterfaceArrayFloat
 ]
 usedDynamicInputs [&AddVectorToPosition]
 usedModules [
  &self:SystemSpawnScript.SystemScriptSource.SystemScriptGraph.NiagaraNodeAssignment_0.SetVariables_#1
  &self:SystemSpawnScript.SystemScriptSource.SystemScriptGraph.NiagaraNodeAssignment_4.SetVariables_#2
  &self:SystemSpawnScript.SystemScriptSource.SystemScriptGraph.NiagaraNodeAssignment_1.SetVariables_#3
  &self:SystemSpawnScript.SystemScriptSource.SystemScriptGraph.NiagaraNodeAssignment_10.SetVariables_#4
  &self:SystemSpawnScript.SystemScriptSource.SystemScriptGraph.NiagaraNodeAssignment_7.SetVariables_#5
  &self:SystemSpawnScript.SystemScriptSource.SystemScriptGraph.NiagaraNodeAssignment_8.SetVariables_#6
  &SystemState
  &self:SystemSpawnScript.SystemScriptSource.SystemScriptGraph.NiagaraNodeAssignment_9.SetVariables_#7
 ]
}
emitters {
 BURST_ {
  emitterUpdateScript [
   &EmitterState(Life Cycle Mode=::NewEnumerator1, Loop Behavior=::NewEnumerator1, Loop Duration=3, UseLoopDelay=false)
   &SpawnBurst_Instantaneous(Spawn Group=5, Spawn Time=1, Use Spawn Probability=false, UseLoopCountLimit=false) {module=SpawnBurst_Instantaneous005}
   &SpawnBurst_Instantaneous(Spawn Count=7, Spawn Group=4, Spawn Time=1, Use Spawn Probability=false, UseLoopCountLimit=false) {module=SpawnBurst_Instantaneous004}
   &SpawnBurst_Instantaneous(
    Spawn Count=&RandomRangeInt(Minimum=15, Maximum=20)
    Spawn Group=3
    Spawn Time=1
    Use Spawn Probability=false
    UseLoopCountLimit=false
   ) {module=SpawnBurst_Instantaneous003}
   &SpawnBurst_Instantaneous(
    Spawn Count=&RandomRangeInt(Minimum=20, Maximum=35)
    Spawn Group=2
    Spawn Time=1
    Use Spawn Probability=false
    UseLoopCountLimit=false
   ) {module=SpawnBurst_Instantaneous002}
   &SpawnBurst_Instantaneous(Spawn Count=7, Spawn Group=1, Spawn Time=1, Use Spawn Probability=false, UseLoopCountLimit=false) {module=SpawnBurst_Instantaneous001}
   &SpawnBurst_Instantaneous(Spawn Time=1, Use Spawn Probability=false, UseLoopCountLimit=false)
  ]
  particleSpawnScript [
   /SetVariables(Particles.VisibilityTag=&SpawnGroup()) {isSetParameters=true}
   &InitializeParticle(
    Color=&FNTimeOfDayColor(Night Color=<rgba 0.6205 0.5885 1 1>, Morning Color=<rgba 1 1 1 1>, Day Color=<rgba 1 1 1 1>, Evening Color=<rgba 1 1 1 1>)
    Lifetime=&MergeHelper_FloatSwitch(
     0=&RandomRangeFloat(Minimum=0.3, Maximum=0.35)
     1=&RandomRangeFloat(Minimum=0.25, Maximum=0.45)
     2=&RandomRangeFloat(Minimum=0.15, Maximum=0.35)
     3=&FN_FrontEndFloatModifier(
      Gameplay=&RandomRangeFloat(Minimum=0.3, Maximum=0.7)
      Front End=&RandomRangeFloat(Minimum=0.15, Maximum=0.5)
     )
     4=&RandomRangeFloat(Minimum=0.125, Maximum=0.5)
     Index=&SpawnGroup()
     Num=::NewEnumerator3
    )
    Position=$System.smoke burst position
    Position Mode=::NewEnumerator1
    Sprite Size Mode=::NewEnumerator3
    UsePositionOffset=false
    Write Material Random=true
   )
   &CameraOffset(Camera Offset Amount=&MergeHelper_FloatSwitch(0=20, 1=2, 2=3, 3=0, 4=0, Index=&SpawnGroup(), Num=::NewEnumerator3))
   &AddVelocity(
    Velocity=&MergeHelper_VectorSwitch(
     0=<0 0 0>
     1=&FN_FrontEndVector3Modifier(
      Gameplay=&RandomRangeVector(Minimum=<-150 -150 -150>, Maximum=<150 150 150>)
      Front End=&RandomRangeVector(Minimum=<-120 -120 -120>, Maximum=<120 120 120>)
     )
     2=&FN_FrontEndVector3Modifier(
      Gameplay=&RandomRangeVector(Minimum=<-150 -150 -150>, Maximum=<150 150 150>)
      Front End=&RandomRangeVector(Minimum=<-120 -120 -120>, Maximum=<120 120 120>)
     )
     3=&Multiply_VectorByFloat(
      Vector=&RandomRangeVector(Minimum=<-150 -150 -150>, Maximum=<150 150 150>)
      Float=&RandomRangeFloat(Minimum=0.5, Maximum=1)
     )
     4=&Multiply_VectorByFloat(
      Vector=&RandomRangeVector(Minimum=<-150 -150 -150>, Maximum=<150 150 150>)
      Float=&RandomRangeFloat(Minimum=0.5, Maximum=1)
     )
     5=&Multiply_VectorByFloat(
      Vector=&RandomRangeVector(Minimum=<50 0 50>, Maximum=<150 0 120>)
      Float=&FN_FrontEndFloatModifier(Gameplay=&RandomRangeFloat(Minimum=0.5, Maximum=1), Front End=0.75)
     )
     Index=&SpawnGroup()
     Num=::NewEnumerator4
    )
    Velocity Speed Scale=&MergeHelper_FloatSwitch(
     0=0
     1=&RandomRangeFloat(Minimum=0.5, Maximum=1)
     2=1
     3=0.8
     4=1
     5=1.5
     Index=&SpawnGroup()
     Num=::NewEnumerator4
    )
   )
   &ShapeLocation(
    Hemisphere Distribution=<0.5 1>
    MaskSpawnGroup=false
    Override Seed=false
    Rotation Angle=0.75
    Rotation Axis=<0 0 1>
    Rotation Mode=::NewEnumerator1
    Sphere Radius=&MergeHelper_FloatSwitch(0=0, 1=1.5, 2=1, 3=1, 4=5, 5=10, Index=&SpawnGroup(), Num=::NewEnumerator4)
    Sphere Surface Distribution=0.5
    UseOwnerScale=true
   )
  ]
  particleUpdateScript [
   &ParticleState()
   /SetVariables(
    Particles.0 - Toon Smoke DynParam 0:Vector4f=&MakeVector4(
     X=&FloatFromCurve(FloatCurve=DataInterface() {propertyValues {Curve {keys [<at 0 0>,<at 1 1>]}}})
     Y=1
     Z=0
     W=0
    )
    Particles.Toon smoke Sprite Size:Vector2f=&Multiply_Vector2D_ByFloat(Float=&FN_FrontEndFloatModifier(Gameplay=1, Front End=0.75), Vector2D=<50 50>)
    Particles.toon smoke color 1:LinearColor=&MakeLinearColorFromVectorAndFloat(
     "Vector (RGB)"=&MakeVectorFromLinearColorRGB(Color=&ColorFromCurve(ColorCurve=DataInterface() {propertyValues {
         AlphaCurve {keys [<at 0 1 RCIM_Cubic>,<at 1 1 RCIM_Cubic>]}
         BlueCurve {keys [<at 0.0012 3.8249>]}
         GreenCurve {keys [<at 0.0012 3.928>]}
         RedCurve {keys [<at 0.0012 4>]}
        }}))
     "Float (Alpha)"=&FloatFromCurve(FloatCurve=DataInterface() {propertyValues {Curve {keys [<at 0 1>,<at 0.8091 0.9932>,<at 1 0>]}}})
    )
   ) {isSetParameters=true}
   &Drag(
    Drag=&MergeHelper_FloatSwitch(
     0=0
     1=&RandomRangeFloat(Minimum=5, Maximum=8)
     2=&RandomRangeFloat(Minimum=0, Maximum=2)
     3=&RandomRangeFloat(Minimum=0, Maximum=2)
     4=&RandomRangeFloat(Minimum=0.8, Maximum=2)
     5=&RandomRangeFloat(Minimum=0.8, Maximum=1.2)
     Index=&SpawnGroup()
     Num=::NewEnumerator4
    )
    Use Linear Drag=true
    Use Rotational Drag=true
   )
   &GravityForce(Gravity=&MergeHelper_VectorSwitch(
     0=<0 0 0>
     1=<0 0 0>
     2=<0 0 0>
     3=<0 0 -350>
     4=&RandomRangeVector(Minimum=<0 0 -225>, Maximum=<1 1 -400>)
     5=&RandomRangeVector(Minimum=<0 0 -350>, Maximum=<1 1 -400>)
     Index=&SpawnGroup()
     Num=::NewEnumerator4
    ))
   /SetVariables(
    Particles.Toon Cloud DynParam:Vector4f=&MakeVector4(
     X=&FloatFromCurve(FloatCurve=DataInterface() {propertyValues {Curve {keys [
          {arriveTangent=0.5561, interpMode=RCIM_Cubic, leaveTangent=0.5561, tangentMode=RCTM_User, time=0, value=0.2}
          {arriveTangent=0.9173, interpMode=RCIM_Cubic, leaveTangent=0.9173, tangentMode=RCTM_User, time=1, value=1}
         ]}}})
     Y=&SelectFloatFromArray(
      Array Sampling Mode=::NewEnumerator0
      Direct Array Index=?
      Float Selection Array=DataInterface() {propertyValues {FloatData [0,1]}}
      Random Seed=0
     )
     Z=0
     W=0
    )
    Particles.Toon Cloud SpriteSize:Vector2f=&Vector2DFromFloat(Value=&FloatFromCurve(
      FloatCurve=DataInterface() {propertyValues {Curve {keys [
          {arriveTangent=3.3458, interpMode=RCIM_Cubic, leaveTangent=3.3458, tangentMode=RCTM_User, time=0, value=1}
          {arriveTangent=1.251, tangentMode=RCTM_User, time=1, value=3}
         ]}}}
      Scale Curve=&Multiply_Float(
       A=&FN_FrontEndFloatModifier(
        Gameplay=&RandomRangeFloat(Minimum=0.1, Maximum=0.7)
        Front End=&RandomRangeFloat(Minimum=0.2, Maximum=0.5)
       )
       B=&RandomRangeFloat(Minimum=10, Maximum=20)
      )
     ))
    Particles.Toon cloud Color:LinearColor=&MakeLinearColorFromVectorAndFloat(
     "Vector (RGB)"=&MakeVectorFromLinearColorRGB(Color=&ColorFromCurve(ColorCurve=DataInterface() {propertyValues {
         AlphaCurve {keys [<at 0 1 RCIM_Cubic>,<at 1 1 RCIM_Cubic>]}
         BlueCurve {keys [<at 0.0012 3.8249>]}
         GreenCurve {keys [<at 0.0012 3.928>]}
         RedCurve {keys [<at 0.0012 4>]}
        }}))
     "Float (Alpha)"=&FloatFromCurve(FloatCurve=DataInterface() {propertyValues {Curve {keys [<at 0 1>,<at 0.8091 0.9932>,<at 1 0>]}}})
    )
   ) {isSetParameters=true}
   /SetVariables(
    Particles.White streaks color:LinearColor=&MakeLinearColorFromVectorAndFloat(
     "Vector (RGB)"=&MakeVectorFromLinearColorRGB(Color=<rgba 1 1 1 1>)
     "Float (Alpha)"=&FloatFromCurve(FloatCurve=DataInterface() {propertyValues {Curve {keys [<at 0 1>,<at 1 0>]}}})
    )
    Particles.white streaks size:Vector2f=&Multiply_Vector2D_ByFloat(
     Float=&FloatFromCurve(FloatCurve=DataInterface() {propertyValues {Curve {defaultValue=1, keys [<at 0 1>,<at 0.273 1>,<at 1 0>]}}})
     Vector2D=&RandomRangeVector2D(Minimum=<1 2.5>, Maximum=<1 3>)
    )
   ) {isSetParameters=true}
   /SetVariables(
    Particles.Droplet Colors:LinearColor=&MakeLinearColorFromVectorAndFloat("Vector (RGB)"=&MakeVectorFromLinearColorRGB(Color=$System.DropletColors), "Float (Alpha)"=1)
    Particles.Droplet SpriteSize:Vector2f=&Multiply_Vector2D_ByFloat(
     Float=&FN_FrontEndFloatModifier(Gameplay=1.5, Front End=1)
     Vector2D=&RandomRangeVector2D(Minimum=<0.5 1.5>, Maximum=<0.7 7.5>)
    )
   ) {isSetParameters=true}
   &SolveForcesAndVelocity(Clamp Velocity=false, Limit Acceleration=false)
   &ScaleSpriteSizeBySpeed(
    Max Scale Factor=&MergeHelper_Vector2Switch(0=<1 1>, 1=<1 1>, 2=<0.5 7>, 3=<0.9 5>, 4=<0 0>, Index=&SpawnGroup(), Num=::NewEnumerator3)
    Min Scale Factor=&MergeHelper_Vector2Switch(0=<1 1>, 1=<1 1>, 2=<0.5 4>, 3=<0.7 4>, 4=<0 0>, Index=&SpawnGroup(), Num=::NewEnumerator3)
    Sample Scale Factor By Curve=true
    Scale Factor Curve=DataInterface() {propertyValues {Curve {defaultValue=0, keys [{tangentMode=RCTM_User, time=0, value=1}, {tangentMode=RCTM_User, time=1, value=0}]}}}
    Velocity Threshold=500
   )
   &SubUVAnimation(
    End Frame Range Override=8
    Evaluation Type=::NewEnumerator0
    Output Alternate Renderer Binding=true
    "Sprite Renderer "=DataInterface() {propertyValues {SpriteRenderer=&self:BURST__0.NiagaraSpriteRendererProperties_4}}
    Start Frame Range Override=4
    SubUV Animation Mode:ENiagara_SubUVLookupModeV2=::NewEnumerator2
    UseEndFrame=true
    UseStartFrame=true
   )
   /SetVariables(
    Particles.Shards Sprite Size:Vector2f=&Vector2DFromFloat(Value=&Multiply_Float(
      A=&FloatFromCurve(
       FloatCurve=DataInterface() {propertyValues {Curve {
          defaultValue=1
          keys [
           {arriveTangent=9.0725, interpMode=RCIM_Cubic, leaveTangent=9.0725, tangentMode=RCTM_User, time=0, value=0}
           {
            arriveTangent=-0.0033
            arriveTangentWeight=0.0469
            interpMode=RCIM_Cubic
            leaveTangent=-0.0033
            tangentMode=RCTM_User
            time=0.1408
            value=0.9847
           }
           {arriveTangent=-0.0405, arriveTangentWeight=0.2864, interpMode=RCIM_Cubic, leaveTangent=-0.0405, tangentMode=RCTM_User, time=1, value=0}
          ]
         }}}
       Scale Curve=&FN_FrontEndFloatModifier(Gameplay=1, Front End=0.75)
      )
      B=&RandomRangeFloat(Minimum=6, Maximum=13)
     ))
    Particles.shards SpriteRotation:NiagaraFloat=&RandomRangeFloat(Minimum=60, Maximum=-60)
   ) {isSetParameters=true}
   &SubUVAnimation(
    Direct Frame Index=8
    Output Alternate Renderer Binding=true
    "Sprite Renderer "=DataInterface() {propertyValues {SpriteRenderer=&self:BURST__0.NiagaraSpriteRendererProperties_5}}
    SubUV Animation Mode:ENiagara_SubUVLookupModeV2=::NewEnumerator5
   ) {module=SubUVAnimation001}
   /SetVariables(Particles.Handle sprite size:Vector2f=&Vector2DFromFloat(Value=&Multiply_Float(
      A=&FloatFromCurve(FloatCurve=DataInterface() {propertyValues {Curve {keys [
           <at 0 1>
           {
            arriveTangent=-0.0033
            arriveTangentWeight=0.0469
            interpMode=RCIM_Cubic
            leaveTangent=-0.0033
            leaveTangentWeight=0.2864
            tangentMode=RCTM_User
            time=0.1408
            value=0.9847
           }
           <at 1 0>
          ]}}})
      B=&FN_FrontEndFloatModifier(Gameplay=13, Front End=9)
     ))) {isSetParameters=true}
   &self:GimbalLock_07(
    Gimbal Lock Direction:NiagaraPosition=&ReturnCameraVector(Camera Vectors=::NewEnumerator4, Coordinate Space=::Local, Origin Position=$Particles.Position)
    Position:NiagaraPosition=$Particles.Position
   ) {module=GimbalLock_07}
  ]
  rendererClasses [&NiagaraSpriteRendererProperties]
  renderers [
   {rendererClass=&NiagaraSpriteRendererProperties}
   {rendererClass=&NiagaraSpriteRendererProperties, rendererIndex=1}
   {rendererClass=&NiagaraSpriteRendererProperties, rendererIndex=2}
   {rendererClass=&NiagaraSpriteRendererProperties, rendererIndex=3}
   {rendererClass=&NiagaraSpriteRendererProperties, rendererIndex=4}
   {rendererClass=&NiagaraSpriteRendererProperties, rendererIndex=5}
  ]
 }
 Intro {
  emitterSpawnScript [
   &InitializeAudioPlayer(Audio Player=DataInterface() {propertyValues {
      Attenuation=null
      Concurrency=null
      MaxPlaysPerTick=10
      SoundToPlay=/CosmeticCompanions/Assets/Audio_Shared/Interaction/Petting/SC_Mimosa_Interaction_React_Long
      bAllowLoopingOneShotSounds=false
      bLimitPlaysPerTick=true
      bOnlyActiveDuringGameplay=false
      bStopWhenComponentIsDestroyed=true
      ConfigurationUserParameter {parameter {name=None, typeDefHandle {registeredTypeIndex=1}, varData []}}
      ParameterNames []
     }})
   /Niagara/Modules/Audio/PlayAudio() {module=PlayAudio001}
   /SetVariables(Emitter.TOD:NiagaraFloat=&FNTimeOfDayScalar(Night Scale=0.5, Morning Scale=0.5, Day Scale=0.5, Evening Scale=0.5)) {isSetParameters=true}
  ]
  emitterUpdateScript [
   &EmitterState(Life Cycle Mode=::NewEnumerator1, Loop Behavior=::NewEnumerator1, Loop Delay=0.05, Loop Duration=1.5, UseLoopDelay=true)
   &SpawnBurst_Instantaneous(Use Spawn Probability=false, UseLoopCountLimit=false)
  ]
  particleSpawnScript [&InitializeParticle(
    Lifetime=$System.IntroLength
    Sprite Rotation Mode=::NewEnumerator0
    Sprite Size Mode=::NewEnumerator3
    Uniform Sprite Size=18
    UsePositionOffset=false
    Write Material Random=true
   )]
  particleUpdateScript [
   &ParticleState()
   &SubUVAnimation(
    Output Alternate Renderer Binding=true
    "Sprite Renderer "=DataInterface() {propertyValues {SpriteRenderer=&self:Intro_0.NiagaraSpriteRendererProperties_1}}
    UseEndFrame=false
    UseStartFrame=false
   )
   &ScaleSpriteSize(
    Scale Sprite Size Mode=::NewEnumerator0
    Uniform Scale Factor=&Add_Float(
     A=0.5
     B=&FloatFromCurve(
      FloatCurve=DataInterface() {propertyValues {Curve {
         defaultValue=0
         keys [
          {arriveTangent=0.4405, interpMode=RCIM_Cubic, leaveTangent=0.4405, tangentMode=RCTM_User, time=0.118, value=1}
          {arriveTangent=-0.9053, interpMode=RCIM_Cubic, leaveTangent=-0.9053, tangentMode=RCTM_User, time=0.25, value=2}
          {arriveTangent=-0.1202, interpMode=RCIM_Cubic, leaveTangent=-0.1202, tangentMode=RCTM_User, time=0.3, value=1}
         ]
        }}}
      Scale Curve=&FN_FrontEndFloatModifier(Gameplay=1, Front End=0.5)
     )
    )
   ) {module=ScaleSpriteSize002}
   &ScaleColor(
    Scale Alpha=&FloatFromCurve(FloatCurve=DataInterface() {propertyValues {Curve {
        defaultValue=0
        keys [
         {arriveTangent=9.0725, interpMode=RCIM_Cubic, leaveTangent=9.0725, tangentMode=RCTM_User, time=0, value=0}
         {arriveTangent=-0.2441, time=0.0108, value=0.9847}
         {arriveTangent=-1.146, leaveTangent=-1.146, time=0.9244, value=1}
         {arriveTangent=-0.0405, interpMode=RCIM_Cubic, leaveTangent=-0.0405, tangentMode=RCTM_User, time=1, value=0}
        ]
       }}})
    Scale RGB=&FN_FrontEndVector3Modifier(Gameplay=<0.7 0.7 0.7>, Front End=<0.5 0.5 0.5>)
    ScaleA=true
    ScaleRGB=true
   )
   &OffsetPosition(Position Offset=&Lerp_Vector(
     A=$System.HeartIntroSpawnPosition
     B=$System.BurstPosition
     Alpha=&FloatFromCurve(FloatCurve=DataInterface() {propertyValues {Curve {keys [
          {arriveTangent=0.7913, interpMode=RCIM_Cubic, leaveTangent=0.7913, tangentMode=RCTM_User, time=0, value=0}
          {arriveTangent=-0.1074, tangentMode=RCTM_User, time=0.0583, value=1}
         ]}}})
    ))
   &DynamicMaterialParameters(
    Index 0 Param 1=&FloatFromCurve(FloatCurve=DataInterface() {propertyValues {Curve {keys [{arriveTangent=0.6296, interpMode=RCIM_Cubic, leaveTangent=0.6296, tangentMode=RCTM_User, time=0, value=0}, <at 0.75 0.9676>, <at 1 1>]}}})
    Index 0 Param 2=0.2
    Module.Index0Float0Write=true
    Module.Index0Float1Write=true
    Module.Index0Float2Write=true
    Module.Index0Float3Write=true
   )
  ]
  rendererClasses [&NiagaraSpriteRendererProperties]
  renderers [{rendererClass=&NiagaraSpriteRendererProperties}]
 }
 Ribbons001 {
  emitterUpdateScript [
   &EmitterState(Life Cycle Mode=::NewEnumerator1, Loop Behavior=::NewEnumerator1, Loop Duration=0.35, UseLoopDelay=false)
   &SpawnRate(SpawnRate=100, Use Spawn Probability=false)
  ]
  particleSpawnScript [
   &InitializeParticle(
    Color=&MakeLinearColorFromVectorAndFloat("Vector (RGB)"=&MakeVectorFromLinearColorRGB(Color=$System.SteamColor), "Float (Alpha)"=1)
    Lifetime=0.3
    Ribbon Width=17.3688
    Ribbon Width Mode=::NewEnumerator1
    UsePositionOffset=false
    Write Material Random=true
   )
   &AddVelocity(Velocity=<0 50 0>)
   &PartitionParticles(
    Number Of Partitions=1
    Particles Per Partition=10
    Partition Mode:ENiagara_PartitionMode_v2=::NewEnumerator1
    Write Partition Index to Ribbon Link Order:ENiagara_AttributeSamplingApplyOutput=::NewEnumerator0
    Write Partition to Ribbon ID=::NewEnumerator0
   ) {enabled=false}
  ]
  particleUpdateScript [
   &ParticleState()
   &ScaleColor(
    Scale Alpha=&FloatFromCurve(
     FloatCurve=DataInterface() {propertyValues {Curve {keys [
         {interpMode=RCIM_Cubic, leaveTangentWeight=1.4087, tangentMode=RCTM_User, time=0, value=0}
         <at 0.2 1 RCIM_Cubic>
         {arriveTangent=-1.8074, interpMode=RCIM_Cubic, leaveTangent=-1.8074, tangentMode=RCTM_User, time=1, value=0}
        ]}}}
     Scale Curve=0.7
    )
    Scale RGB=&MakeVectorFromLinearColorRGB(Color=&ColorFromCurve(ColorCurve=DataInterface() {propertyValues {
        AlphaCurve {keys []}
        BlueCurve {keys [<at 0.35 0.5469 RCIM_Cubic>]}
        GreenCurve {keys [<at 0.35 0.8382 RCIM_Cubic>]}
        RedCurve {keys [<at 0.35 0.949 RCIM_Cubic>]}
       }}))
    ScaleA=true
    ScaleRGB=true
   )
   &ScaleRibbonWidth(Ribbon Width Scale=&FloatFromCurve(
     FloatCurve=DataInterface() {propertyValues {Curve {
        defaultValue=0
        keys [
         {interpMode=RCIM_Cubic, tangentMode=RCTM_User, time=0, value=1}
         {interpMode=RCIM_Cubic, tangentMode=RCTM_User, time=0.5, value=1}
         {tangentMode=RCTM_User, time=0.9292, value=0.3301}
        ]
       }}}
     Scale Curve=0.4
    ))
   /SetVariables(Particles.Position=&LerpPosition(
     A=Linked() {linkedVariable {name=System.BurstPosition, type=Vector3f}}
     B=&LerpPosition(A=Linked() {linkedVariable {name=System.Spout position, type=Vector3f}}, B=$System.Latte heart Position, Alpha=-2.4859)
     Alpha=&FloatFromCurve(FloatCurve=DataInterface() {propertyValues {Curve {keys [<at 0 1>,<at 0.4 0.1906>,<at 0.55 0>]}}})
    )) {isSetParameters=true}
   &DynamicMaterialParameters(
    Index 0 Param 1=&FloatFromCurve(FloatCurve=DataInterface() {propertyValues {Curve {keys [<at 0.1 1>,<at 1 0>]}}}, Scale Curve=0.6327)
    Index 0 Param 2=&FloatFromCurve(FloatCurve=DataInterface() {propertyValues {Curve {keys [<at 0 0 RCIM_Cubic>,<at 0.1 1 RCIM_Cubic>,<at 0.5 0 RCIM_Cubic>]}}})
    Index 0 Param 3=-1.5746
    Index 0 Param 4=36.103
    Module.Param0WriteEnabled=true
    Module.Param1WriteEnabled=true
    Module.Param2WriteEnabled=true
    Module.Param3WriteEnabled=true
   )
   &Drag(Use Linear Drag=true, Use Rotational Drag=true)
   &SolveForcesAndVelocity(Clamp Velocity=false, Limit Acceleration=false)
   /SetVariables(
    Particles.Ribbon coffee width:NiagaraFloat=&FloatFromCurve(FloatCurve=DataInterface() {propertyValues {Curve {keys [<at 0 1>,<at 0.63 0>]}}}, Scale Curve=15)
    Particles.coffee color:LinearColor=&MakeLinearColorFromVectorAndFloat("Vector (RGB)"=&MakeVectorFromLinearColorRGB(Color=<rgba 0.4479 0.2959 0.1212 1>), "Float (Alpha)"=1)
   ) {isSetParameters=true}
   &self:GimbalLock_02(
    Gimbal Lock Direction:NiagaraPosition=&ReturnCameraVector(Camera Vectors=::NewEnumerator4, Coordinate Space=::Local, Origin Position=$Particles.Position)
    Position:NiagaraPosition=$Particles.Position
   ) {module=GimbalLock_02}
  ]
  rendererClasses [&NiagaraRibbonRendererProperties]
  renderers [{rendererClass=&NiagaraRibbonRendererProperties}]
 }
 SplashElements_feet {
  emitterSpawnScript [/SetVariables(Emitter.highlight=DataInterface() {propertyValues {GpuSyncMode=SyncCpuToGpu, MaxElements=0, ColorData [<rgba 1 1 1 1>]}}) {isSetParameters=true}]
  emitterUpdateScript [
   &EmitterState(Life Cycle Mode=::NewEnumerator1, Loop Behavior=::NewEnumerator1, Loop Delay=0.5, Loop Duration=0.54, UseLoopDelay=false)
   &SpawnRate(SpawnRate=75, Use Spawn Probability=false)
  ]
  particleSpawnScript [
   &InitializeParticle(
    AdjustAlpha=false
    AdjustHue=true
    AdjustSaturation=false
    AdjustValue=false
    Color=<rgba 0.7305 0.3005 0.1022 1>
    Color Mode=::NewEnumerator3
    Lifetime Max=0.65
    Lifetime Min=0.2
    Lifetime Mode=::NewEnumerator1
    Mass Mode=::NewEnumerator3
    Mesh Scale=&Multiply_VectorByFloat(Vector=<2 1 1>, Float=&RandomRangeFloat(Minimum=0.3, Maximum=2))
    Mesh Scale Mode=::NewEnumerator4
    Position Offset=&FN_FrontEndVector3Modifier(Gameplay=<0 0 15>, Front End=<0 0 3>)
    Position Offset Coordinate Space=::Local
    Sprite Size=&RandomRangeVector2D(Minimum=<1.9792 5>, Maximum=<3 5>)
    Sprite Size Mode=::NewEnumerator4
    UsePositionOffset=true
    Write Material Random=true
   )
   &SubUVAnimation(End Frame=3, SubUV Animation Mode:ENiagara_SubUVLookupMode=::NewEnumerator2)
   &AddVelocity(
    Rotation Coordinate Space=::Simulation
    Velocity=&RandomRangeVector(Minimum=<-120 -125 35>, Maximum=<120 120 85>)
   )
   &ShapeLocation(
    MaskSpawnGroup=false
    Offset=$System.BurstPosition
    Offset Coordinate Space=::Local
    Offset Mode=::NewEnumerator0
    Sphere Distribution=::NewEnumerator1
    Sphere Radius=1
    UseOwnerScale=true
   )
  ]
  particleUpdateScript [
   &ParticleState()
   &GravityForce(
    Coordinate Space=::Local
    Gravity=&MakeVector(X=0, Y=0, Z=&RandomRangeFloat(Minimum=-150, Maximum=-500))
   )
   &ScaleColor(
    Scale Alpha=&FloatFromCurve(FloatCurve=DataInterface() {propertyValues {Curve {keys [<at -0.0046 1.0113 RCIM_Cubic>,<at 1.0011 0.002 RCIM_Cubic>]}}})
    ScaleA=true
    ScaleRGB=true
   )
   &Drag(
    Drag=&RandomRangeFloat(Minimum=4, Maximum=8, Evaluation Type=::NewEnumerator1)
    Use Linear Drag=true
    Use Rotational Drag=false
   )
   &SolveForcesAndVelocity(Clamp Velocity=false, Limit Acceleration=false)
   &ScaleSpriteSize(
    Uniform Curve Scale=&FN_FrontEndFloatModifier(
     Gameplay=&RandomRangeFloat(Minimum=0.5, Maximum=2)
     Front End=&RandomRangeFloat(Minimum=0.2, Maximum=0.8)
    )
    Uniform Curve Sprite Scale=DataInterface() {propertyValues {Curve {keys [
        <at 0 1 RCIM_Cubic>
        {arriveTangent=-0.4286, interpMode=RCIM_Cubic, leaveTangent=-0.4286, time=0.1, value=1}
        {arriveTangent=-1.1111, interpMode=RCIM_Cubic, leaveTangent=-1.1111, time=0.7, value=0.7}
        <at 1 0 RCIM_Cubic>
       ]}}}
   )
   /SetVariables(Particles.Color=&MakeLinearColorFromVectorAndFloat(
     "Vector (RGB)"=&MakeVectorFromLinearColorRGB(Color=&SelectLinearColorFromArray(
       Array Sampling Mode=::NewEnumerator0
       Direct Array Index=?
       Color Selection Array=DataInterface() {propertyValues {ColorData [<rgba 0.949 0.8382 0.5469 1>, <rgba 0.7305 0.3005 0.1022 1>]}}
       Random Seed=0
      ))
     "Float (Alpha)"=1
    )) {isSetParameters=true}
   &self:GimbalLock_06(
    Gimbal Lock Direction:NiagaraPosition=&ReturnCameraVector(Camera Vectors=::NewEnumerator4, Coordinate Space=::Local, Origin Position=$Particles.Position)
    Position:NiagaraPosition=$Particles.Position
   ) {module=GimbalLock_06}
  ]
  rendererClasses [&NiagaraSpriteRendererProperties]
  renderers [{rendererClass=&NiagaraSpriteRendererProperties}]
 }
}
summary {
 systemName=NS_ReactFX_SnugLoaf
 emitters [
  {emitterName=Intro, rendererClasses [&NiagaraSpriteRendererProperties]}
  {emitterName=SplashElements_feet, rendererClasses [&NiagaraSpriteRendererProperties]}
  {emitterName=Ribbons001, rendererClasses [&NiagaraRibbonRendererProperties]}
  {emitterName=BURST_, rendererClasses [&NiagaraSpriteRendererProperties]}
 ]
 userVariables [
  User.bIsFrontEnd {type=NiagaraBool}
  User.bIsFrontEndPreview {type=NiagaraBool}
  User.FrontEndScale {type=NiagaraFloat, defaultValue=0, description=""}
  User.RibbonIDOffset {type=NiagaraInt32}
 ]
}
