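# Niagara reaction effect for cosmetic companions: a spinning, pulsing emblem that bursts into falling pieces, with a petting-reaction sound.
# NE_Heart plays the sound and moves layered sprites and a mesh from the intro position to System.BurstPosition.
# NE_Burst and NE_Heart_Pieces fire just before System.IntroLength ends, adding glows, flickering sparks and debris.
# Spawn-group switches pack several layers into single emitters, while front-end and time-of-day inputs adjust size and brightness.
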NiagaraSystem /CosmeticCompanions/FX/React/NS_ReactFX_SphereDaily
dependencies {
 usedDataInterfaces [
  &NiagaraDataInterfaceAudioPlayer
  &NiagaraDataInterfaceArrayFloat3
  &NiagaraDataInterfaceCurve
  &NiagaraDataInterfaceSpriteRendererInfo
  &NiagaraDataInterfaceArrayFloat
 ]
 usedDynamicInputs [&FN_FrontEndVector2Modifier]
 usedModules [
  &self:SystemSpawnScript.SystemScriptSource.SystemScriptGraph.NiagaraNodeAssignment_0.SetVariables_#1
  &self:SystemSpawnScript.SystemScriptSource.SystemScriptGraph.NiagaraNodeAssignment_1.SetVariables_#2
  &self:SystemSpawnScript.SystemScriptSource.SystemScriptGraph.NiagaraNodeAssignment_4.SetVariables_#3
  &self:SystemSpawnScript.SystemScriptSource.SystemScriptGraph.NiagaraNodeAssignment_2.SetVariables_#4
  &self:SystemSpawnScript.SystemScriptSource.SystemScriptGraph.NiagaraNodeAssignment_5.SetVariables_#5
  &self:SystemSpawnScript.SystemScriptSource.SystemScriptGraph.NiagaraNodeAssignment_6.SetVariables_#6
  &SystemState
 ]
}
emitters {
 LWE_AirGlow {rendererClasses [], renderers []}
 LWE_Smoke {rendererClasses [], renderers []}
 LWE_Sparkles {rendererClasses [], renderers []}
 NE_Burst {
  emitterUpdateScript [
   &EmitterState(Life Cycle Mode=::NewEnumerator1, Loop Behavior=::NewEnumerator1, Loop Duration=1.5, UseLoopDelay=false)
   &SpawnBurst_Instantaneous(
    Spawn Time=&Subtract_Float(A=$System.IntroLength, B=0.13)
    Use Spawn Probability=false
    UseLoopCountLimit=false
   )
   &SpawnBurst_Instantaneous(
    Spawn Group=1
    Spawn Time=&Subtract_Float(A=$System.IntroLength, B=0.13)
    Use Spawn Probability=false
    UseLoopCountLimit=false
   ) {module=SpawnBurst_Instantaneous001}
   &SpawnBurst_Instantaneous(
    Spawn Group=2
    Spawn Time=&Subtract_Float(A=$System.IntroLength, B=0.13)
    Use Spawn Probability=false
    UseLoopCountLimit=false
   ) {module=SpawnBurst_Instantaneous002}
   &SpawnBurst_Instantaneous(
    Spawn Group=3
    Spawn Time=&Subtract_Float(A=$System.IntroLength, B=0.13)
    Use Spawn Probability=false
    UseLoopCountLimit=false
   ) {module=SpawnBurst_Instantaneous003}
   &SpawnBurst_Instantaneous(
    Spawn Count=15
    Spawn Group=4
    Spawn Time=&Subtract_Float(A=$System.IntroLength, B=0.13)
    Use Spawn Probability=false
    UseLoopCountLimit=false
   ) {module=SpawnBurst_Instantaneous004}
   &SpawnBurst_Instantaneous(
    Spawn Group=5
    Spawn Time=&Subtract_Float(A=$System.IntroLength, B=0.13)
    Use Spawn Probability=false
    UseLoopCountLimit=false
   ) {module=SpawnBurst_Instantaneous005}
  ]
  particleSpawnScript [
   /SetVariables(Particles.VisibilityTag=&SpawnGroup()) {isSetParameters=true}
   &InitializeParticle(
    Color=&MergeHelper_ColorSwitch(
     0=<rgba 5 4.1309 1.0022 0.1>
     1=<rgba 5 3.0011 1.0022 0.1>
     2=<rgba 5 3.0011 1.0022 0.1>
     3=<rgba 5 4.1309 1.0022 0.1>
     4=<rgba 5 4.1309 1.0022 0.1>
     5=<rgba 5 3.3343 1.0022 0.5>
     Index=$Particles.VisibilityTag
     Num=::NewEnumerator4
    )
    Lifetime=&MergeHelper_FloatSwitch(
     0=0.2
     1=0.3
     2=0.3
     3=0.3
     4=&RandomRangeFloat(Minimum=0.4, Maximum=0.8)
     5=0.5
     Index=$Particles.VisibilityTag
     Num=::NewEnumerator4
    )
    Position Offset=$System.BurstPosition
    Sprite Size Mode=::NewEnumerator3
    Uniform Sprite Size=&MergeHelper_FloatSwitch(
     0=&FN_FrontEndFloatModifier(Gameplay=100, Front End=70)
     1=&FN_FrontEndFloatModifier(Gameplay=45, Front End=30)
     2=&FN_FrontEndFloatModifier(Gameplay=45, Front End=30)
     3=&FN_FrontEndFloatModifier(Gameplay=25, Front End=15)
     4=&FN_FrontEndFloatModifier(
      Gameplay=&RandomRangeFloat(Minimum=10, Maximum=20)
      Front End=&RandomRangeFloat(Minimum=8, Maximum=16)
     )
     5=&FN_FrontEndFloatModifier(Gameplay=35, Front End=25)
     Index=$Particles.VisibilityTag
     Num=::NewEnumerator4
    )
    UsePositionOffset=true
    Write Material Random=true
   )
   &ShapeLocation(
    MaskSpawnGroup=false
    Override Seed=false
    Sphere Radius=&MergeHelper_FloatSwitch(0=0, 1=0, 2=0, 3=0, 4=10, 5=0, Index=$Particles.VisibilityTag, Num=::NewEnumerator4)
    UseOwnerScale=true
   )
   &AddVelocity(
    Module.Use Falloff Distance=false
    Override Seed=false
    Velocity Mode=::NewEnumerator1
    Velocity Speed=&MergeHelper_FloatSwitch(
     0=0
     1=0
     2=0
     3=0
     4=&RandomRangeFloat(Minimum=25, Maximum=30)
     5=0
     Index=$Particles.VisibilityTag
     Num=::NewEnumerator4
    )
   )
  ]
  particleUpdateScript [
   &ParticleState()
   &ScaleSpriteSize(Uniform Curve Scale=2)
   &ScaleColor(
    Scale Alpha=&MergeHelper_FloatSwitch(
     0=&FloatFromCurve(
      FloatCurve=DataInterface() {propertyValues {Curve {keys [<at 0 0>,<at 0.5 1>,<at 1 0>]}}}
      Scale Curve=&FNTimeOfDayScalar(Night Scale=0.5, Morning Scale=1, Evening Scale=1)
     )
     1=&FloatFromCurve(
      FloatCurve=DataInterface() {propertyValues {Curve {keys [<at 0 0>,<at 0.5 1>,<at 1 0>]}}}
      Scale Curve=&FNTimeOfDayScalar(Night Scale=0.5, Morning Scale=1, Evening Scale=1)
     )
     2=&FloatFromCurve(
      FloatCurve=DataInterface() {propertyValues {Curve {keys [<at 0 0>,<at 0.5 1>,<at 1 0>]}}}
      Scale Curve=&FNTimeOfDayScalar(Night Scale=0.5, Morning Scale=1, Evening Scale=1)
     )
     3=&FloatFromCurve(
      FloatCurve=DataInterface() {propertyValues {Curve {keys [<at 0 0>,<at 0.5 1>,<at 1 0>]}}}
      Scale Curve=&FNTimeOfDayScalar(Night Scale=0.5, Morning Scale=1, Evening Scale=1)
     )
     4=&FloatFromCurve(
      FloatCurve=DataInterface() {propertyValues {Curve {keys [<at 0 0>,<at 0.1 1>,<at 0.2 0>,<at 0.3 1>,<at 0.4 0>,<at 0.5 1>,<at 0.6 0>,<at 0.7 1>,<at 0.8 0>,<at 0.9 1>,<at 1 0>]}}}
      Scale Curve=&FNTimeOfDayScalar(Night Scale=0.5, Morning Scale=1, Evening Scale=1)
     )
     5=&FloatFromCurve(
      FloatCurve=DataInterface() {propertyValues {Curve {keys [<at 0 0>,<at 0.5 1>,<at 1 0>]}}}
      Scale Curve=&FNTimeOfDayScalar(Night Scale=0.5, Morning Scale=1, Evening Scale=1)
     )
     Index=$Particles.VisibilityTag
     Num=::NewEnumerator4
    )
    ScaleA=true
    ScaleRGB=true
   )
   &Drag(
    Drag=&MergeHelper_FloatSwitch(
     0=0
     1=0
     2=0
     3=0
     4=&RandomRangeFloat(Minimum=3, Maximum=6)
     5=0
     Index=$Particles.VisibilityTag
     Num=::NewEnumerator4
    )
    Use Linear Drag=true
    Use Rotational Drag=true
   )
   &CurlNoiseForce(
    Bias Noise=false
    Noise Frequency=&MergeHelper_FloatSwitch(0=0, 1=0, 2=0, 3=0, 4=50, 5=0, Index=$Particles.VisibilityTag, Num=::NewEnumerator4)
    Noise Strength=&MergeHelper_FloatSwitch(0=0, 1=0, 2=0, 3=0, 4=25, 5=0, Index=$Particles.VisibilityTag, Num=::NewEnumerator4)
    Pan Noise=false
   )
   &GravityForce(Gravity=&MergeHelper_VectorSwitch(
     0=<0 0 0>
     1=<0 0 0>
     2=<0 0 0>
     3=<0 0 0>
     4=&RandomRangeVector(Minimum=<0 0 -50>, Maximum=<0 0 -200>)
     5=<0 0 0>
     Index=$Particles.VisibilityTag
     Num=::NewEnumerator4
    ))
   &SolveForcesAndVelocity(Clamp Velocity=false, Limit Acceleration=false)
  ]
  rendererClasses [&NiagaraSpriteRendererProperties]
  renderers [
   {rendererClass=&NiagaraSpriteRendererProperties}
   {rendererClass=&NiagaraSpriteRendererProperties, rendererIndex=1}
   {rendererClass=&NiagaraSpriteRendererProperties, rendererIndex=2}
   {rendererClass=&NiagaraSpriteRendererProperties, rendererIndex=3}
   {rendererClass=&NiagaraSpriteRendererProperties, rendererIndex=4}
   {rendererClass=&NiagaraSpriteRendererProperties, rendererIndex=5}
  ]
 }
 NE_Heart {
  emitterSpawnScript [
   &InitializeAudioPlayer(Audio Player=DataInterface() {propertyValues {
      Attenuation=null
      Concurrency=null
      MaxPlaysPerTick=10
      SoundToPlay=/CosmeticCompanions/Assets/Audio_Shared/Interaction/Petting/SC_Mimosa_Interaction_React_Default
      bAllowLoopingOneShotSounds=false
      bLimitPlaysPerTick=true
      bOnlyActiveDuringGameplay=false
      bStopWhenComponentIsDestroyed=true
      ConfigurationUserParameter {parameter {name=None, typeDefHandle {registeredTypeIndex=1}, varData []}}
      ParameterNames []
     }})
   /Niagara/Modules/Audio/PlayAudio() {module=PlayAudio001}
  ]
  emitterUpdateScript [
   &EmitterState()
   &SpawnBurst_Instantaneous(Use Spawn Probability=false, UseLoopCountLimit=false)
   &SpawnBurst_Instantaneous(Spawn Count=3, Spawn Group=1, Use Spawn Probability=false, UseLoopCountLimit=false) {module=SpawnBurst_Instantaneous001}
   &SpawnBurst_Instantaneous(Spawn Group=3, Use Spawn Probability=false, UseLoopCountLimit=false) {module=SpawnBurst_Instantaneous003}
   &SpawnBurst_Instantaneous(Spawn Group=4, Use Spawn Probability=false, UseLoopCountLimit=false) {module=SpawnBurst_Instantaneous004}
   /SetVariables(Emitter.CoinEmissive:LinearColor=&FNTimeOfDayColor(Night Color=<rgba 0.025 0.0208 0 1>, Morning Color=<rgba 0 0 0 1>, Day Color=<rgba 0 0 0 1>, Evening Color=<rgba 0 0 0 1>)) {isSetParameters=true}
  ]
  particleSpawnScript [
   /SetVariables(Particles.VisibilityTag=&SpawnGroup()) {isSetParameters=true}
   &FN_SetSpawnGroup()
   &InitializeParticle(
    Color=&MergeHelper_ColorSwitch(
     0=<rgba 1 1 1 1>
     1=<rgba 5 2.4111 0.3311 1>
     2=<rgba 5 2.4111 0.3311 1>
     3=&FNTimeOfDayColor(
      Night Color=<rgba 2.5 2.5 0.25 0.5>
      Morning Color=<rgba 5 2.75 0.5 0.5>
      Day Color=<rgba 5 2.75 0.5 0.5>
      Evening Color=<rgba 5 2.75 0.5 0.5>
     )
     4=&FNTimeOfDayColor(
      Night Color=<rgba 2 0.9759 0.2443 1>
      Morning Color=<rgba 2 1.8015 0.2443 1>
      Day Color=<rgba 2 0.9759 0.2443 1>
      Evening Color=<rgba 2 1.8015 0.2443 1>
     )
     Index=$Particles.VisibilityTag
     Num=::NewEnumerator3
    )
    Lifetime=$System.IntroLength
    Mesh Scale=&FN_FrontEndVector3Modifier(Gameplay=<0.1 0.2 0.1>, Front End=<0.075 0.175 0.075>)
    Mesh Scale Mode=::NewEnumerator4
    Position Offset=&MergeHelper_VectorSwitch(
     0=<0 0 0>
     1=&SelectVectorFromArray(Direct Array Index=$Particles.SpawnGroup, Vector Selection Array=DataInterface() {propertyValues {FloatData [<3 0 0>, <-3 0 2>, <0 0 -3>]}})
     2=<0 0 0>
     3=<0 0 2>
     4=<0 0 0>
     Index=$Particles.VisibilityTag
     Num=::NewEnumerator3
    )
    Position Offset Coordinate Space=::Local
    Sprite Rotation Angle=&MergeHelper_FloatSwitch(
     0=45
     1=&RandomRangeFloat(Minimum=-360, Maximum=360)
     2=0
     3=0
     Index=$Particles.VisibilityTag
     Num=::NewEnumerator2
    )
    Sprite Rotation Mode=::NewEnumerator0
    Sprite Size Mode=::NewEnumerator3
    Uniform Sprite Size=&MergeHelper_FloatSwitch(
     0=&FN_FrontEndFloatModifier(Gameplay=18, Front End=12)
     1=&FN_FrontEndFloatModifier(Gameplay=&RandomRangeFloat(Minimum=7, Maximum=14), Front End=5)
     2=&FN_FrontEndFloatModifier(Gameplay=15, Front End=9)
     3=&FN_FrontEndFloatModifier(Gameplay=30, Front End=21.5)
     Index=$Particles.VisibilityTag
     Num=::NewEnumerator2
    )
    UsePositionOffset=true
    Write Material Random=true
   )
  ]
  particleUpdateScript [
   &ParticleState()
   &UpdateMeshOrientation(
    Rotation Rate=&FloatFromCurve(
     FloatCurve=DataInterface() {propertyValues {Curve {defaultValue=1, keys [{tangentMode=RCTM_User, time=0, value=1}, {tangentMode=RCTM_User, time=1, value=0}]}}}
     Scale Curve=3.9
    )
    Rotation Vector=<1 0.5 0.25>
   )
   &SpriteRotationRate(Rotation Rate=&MergeHelper_FloatSwitch(
     0=&FloatFromCurve(
      FloatCurve=DataInterface() {propertyValues {Curve {keys [<at 0 1>,<at 0.5 1>,<at 0.75 0>]}}}
      Scale Curve=&FloatFromCurve(FloatCurve=DataInterface() {propertyValues {Curve {keys [<at 0 1>,<at 0.5 1>,<at 0.6 0.5>]}}}, Scale Curve=580)
     )
     1=0
     2=0
     3=45
     4=0
     Index=$Particles.VisibilityTag
     Num=::NewEnumerator3
    ))
   &ScaleSpriteSize(
    Scale Factor=&MakeVector2D(
     X=&FloatFromCurve(FloatCurve=DataInterface() {propertyValues {Curve {keys [<at 0.5 1>,<at 0.6 1.5>,<at 0.7 1>,<at 0.8 1.5>,<at 0.9 1>,<at 0.95 1>,<at 0.975 3>,<at 1 1>]}}})
     Y=&FloatFromCurve(FloatCurve=DataInterface() {propertyValues {Curve {keys [<at 0.5 1>,<at 0.6 1.5>,<at 0.7 1>,<at 0.8 1.5>,<at 0.9 1>,<at 0.95 1>,<at 0.975 3>,<at 1 1>]}}})
    )
    Scale Sprite Size Mode=::NewEnumerator2
   )
   &ScaleColor(
    Scale Alpha=&MergeHelper_FloatSwitch(
     0=&FloatFromCurve(FloatCurve=DataInterface() {propertyValues {Curve {
         defaultValue=0
         keys [<at 0.35 0>, {arriveTangent=-0.2441, time=0.405, value=1}, {arriveTangent=-1.146, leaveTangent=-1.146, time=1, value=1}]
        }}})
     1=&FloatFromCurve(
      FloatCurve=DataInterface() {propertyValues {Curve {
         defaultValue=0
         keys [
          {arriveTangent=9.0725, interpMode=RCIM_Cubic, leaveTangent=9.0725, tangentMode=RCTM_User, time=0, value=0}
          {arriveTangent=-0.2441, leaveTangent=-0.2441, time=0.1, value=0.9847}
          <at 0.2 0>
          <at 0.3 1>
          <at 0.5 0>
          <at 0.6 1>
          <at 0.8 0>
          <at 0.9 1>
          {arriveTangent=-1.146, leaveTangent=-1.146, time=1, value=0}
         ]
        }}}
      Scale Curve=&FNTimeOfDayScalar(Night Scale=0.5, Morning Scale=1, Evening Scale=1)
     )
     2=&FloatFromCurve(
      FloatCurve=DataInterface() {propertyValues {Curve {
         defaultValue=0
         keys [
          {arriveTangent=9.0725, interpMode=RCIM_Cubic, leaveTangent=9.0725, tangentMode=RCTM_User, time=0, value=0}
          {arriveTangent=-0.2441, leaveTangent=-0.2441, time=0.1408, value=0.9847}
          {arriveTangent=-1.146, leaveTangent=-1.146, time=1, value=1}
         ]
        }}}
      Scale Curve=&FNTimeOfDayScalar(Night Scale=0.5, Morning Scale=1, Evening Scale=1)
     )
     3=&FloatFromCurve(
      FloatCurve=DataInterface() {propertyValues {Curve {defaultValue=0, keys [{arriveTangent=9.0725, time=0, value=0}, <at 0.35 0>, <at 0.405 1>, <at 0.9 1>, <at 1 0 RCIM_Cubic>]}}}
      Scale Curve=&FNTimeOfDayScalar(Night Scale=0.5, Morning Scale=1, Evening Scale=1)
     )
     4=&FloatFromCurve(FloatCurve=DataInterface() {propertyValues {Curve {keys [<at 0 1>, {interpMode=RCIM_Cubic, tangentMode=RCTM_User, time=0.35, value=1}, <at 0.405 0>]}}})
     Index=$Particles.VisibilityTag
     Num=::NewEnumerator3
    )
    ScaleA=true
    ScaleRGB=true
   )
   &OffsetPosition(Position Offset=&Lerp_Vector(
     A=$System.HeartIntroSpawnPosition
     B=$System.BurstPosition
     Alpha=&FloatFromCurve(FloatCurve=DataInterface() {propertyValues {Curve {keys [
          {arriveTangent=0.7913, interpMode=RCIM_Cubic, leaveTangent=0.7913, tangentMode=RCTM_User, time=0, value=0}
          {arriveTangent=-0.1074, tangentMode=RCTM_User, time=0.3, value=1}
         ]}}})
    ))
   &DynamicMaterialParameters(
    Index 0 Param 1=3
    Index 0 Param 2=3
    Index 0 Param 3=3
    Module.Index0Float0Write=true
    Module.Index0Float1Write=true
    Module.Index0Float2Write=true
    Module.Index0Float3Write=true
   )
   &SubUVAnimation(
    "Sprite Renderer "=DataInterface() {propertyValues {SpriteRenderer=&self:NE_Heart_0.NiagaraSpriteRendererProperties_2}}
    SubUV Animation Mode:ENiagara_SubUVLookupModeV2=::NewEnumerator5
   )
   &CameraOffset(Camera Offset Amount=&MergeHelper_FloatSwitch(0=-2, 1=0, 2=0, 3=0, 4=0, Index=$Particles.VisibilityTag, Num=::NewEnumerator3))
  ]
  rendererClasses [&NiagaraSpriteRendererProperties, &NiagaraMeshRendererProperties]
  renderers [
   {rendererClass=&NiagaraSpriteRendererProperties}
   {rendererClass=&NiagaraSpriteRendererProperties, rendererIndex=1}
   {rendererClass=&NiagaraSpriteRendererProperties, rendererIndex=2}
   {rendererClass=&NiagaraMeshRendererProperties, rendererIndex=3}
  ]
 }
 NE_Heart_Pieces {
  emitterUpdateScript [
   &EmitterState()
   &SpawnBurst_Instantaneous(
    Spawn Count=4
    Spawn Time=&Subtract_Float(A=$System.IntroLength, B=0.025)
    Use Spawn Probability=false
    UseLoopCountLimit=false
   )
  ]
  particleSpawnScript [
   &FN_SetSpawnGroup(Number of Spawn Groups=3)
   &InitializeParticle(
    Lifetime Max=1.55
    Lifetime Min=0.35
    Lifetime Mode=::NewEnumerator1
    Position Offset=$System.BurstPosition
    Sprite Rotation Mode=::NewEnumerator0
    Sprite Size Mode=::NewEnumerator3
    Uniform Sprite Size=&FN_FrontEndFloatModifier(Gameplay=15, Front End=6.5)
    UsePositionOffset=true
    Write Material Random=true
   )
   &AddVelocity(
    Module.Use Falloff Distance=false
    Origin Offset=$System.BurstPosition
    Override Seed=false
    Velocity Mode=::NewEnumerator1
    Velocity Speed=&RandomRangeFloat(Minimum=150, Maximum=200)
   )
  ]
  particleUpdateScript [
   &ParticleState()
   &SpriteRotationRate(Rotation Rate=&RandomRangeFloat(Minimum=-360, Maximum=360))
   &DynamicMaterialParameters(
    Index 0 Param 3=10
    Module.Index0Float0Write=true
    Module.Index0Float1Write=true
    Module.Index0Float2Write=true
    Module.Index0Float3Write=true
   )
   &ScaleSpriteSize(
    Scale Sprite Size Mode=::NewEnumerator0
    Uniform Scale Factor=&FloatFromCurve(FloatCurve=DataInterface() {propertyValues {Curve {keys [{arriveTangent=3.3458, time=0, value=1}, <at 0.1 1.5>, <at 0.2 1>, <at 0.5 1>, {arriveTangent=1.251, tangentMode=RCTM_User, time=1, value=0}]}}})
   ) {module=ScaleSpriteSize002}
   &ScaleColor(
    Scale Alpha=&FNTimeOfDayScalar(Night Scale=0.75, Morning Scale=1, Evening Scale=1)
    ScaleA=true
    ScaleRGB=true
   )
   &GravityForce(Gravity=&RandomRangeVector(Minimum=<0 0 -360>, Maximum=<0 0 -1080>))
   &Drag(Drag=&RandomRangeFloat(Minimum=4, Maximum=7), Use Linear Drag=true, Use Rotational Drag=true)
   &SolveForcesAndVelocity(Clamp Velocity=false, Limit Acceleration=false)
   &SubUVAnimation(
    Direct Frame Index:NiagaraFloat=&SelectFloatFromArray(Direct Array Index=&SpawnGroup(), Float Selection Array=DataInterface() {propertyValues {FloatData [2,3,1]}})
    "Sprite Renderer "=DataInterface() {propertyValues {SpriteRenderer=&self:NE_Heart_Pieces_0.NiagaraSpriteRendererProperties_2}}
    SubUV Animation Mode:ENiagara_SubUVLookupModeV2=::NewEnumerator5
   )
  ]
  rendererClasses [&NiagaraSpriteRendererProperties]
  renderers [{rendererClass=&NiagaraSpriteRendererProperties}]
 }
}
summary {
 systemName=NS_ReactFX_SphereDaily
 emitters [
  {emitterName=NE_Heart, rendererClasses [&NiagaraSpriteRendererProperties, &NiagaraMeshRendererProperties]}
  {emitterName=NE_Heart_Pieces, rendererClasses [&NiagaraSpriteRendererProperties]}
  {emitterName=NE_Burst, rendererClasses [&NiagaraSpriteRendererProperties]}
  {emitterName=LWE_Sparkles, rendererClasses []}
  {emitterName=LWE_AirGlow, rendererClasses []}
  {emitterName=LWE_Smoke, rendererClasses []}
 ]
 userVariables [
  User.bIsFrontEnd {type=NiagaraBool}
  User.bIsFrontEndPreview {type=NiagaraBool}
 ]
}
