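# Niagara system for the TalonScone cosmetic companion's react effect: a fiery heart with a petting reaction sound.
# Intro_FireParticles swirl from HeartIntroSpawnPosition to BurstPosition over IntroLength, where Heart and HeartFire fade in.
# HeartFire builds layered flames using partitioned pivot offsets, while FireParticles adds a delayed burst of sparks and embers.
# Spawn groups drive per-layer behavior through switch inputs, and front-end modifiers adjust size and opacity outside gameplay.
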NiagaraSystem /CosmeticCompanions/FX/React/NS_ReactFX_TalonScone
dependencies {
 usedDataInterfaces [
  &NiagaraDataInterfaceAudioPlayer
  &NiagaraDataInterfaceCurve
  &NiagaraDataInterfaceSpriteRendererInfo
  &NiagaraDataInterfaceColorCurve
  &NiagaraDataInterfaceArrayFloat3
  &NiagaraDataInterfaceArrayFloat2
  &NiagaraDataInterfaceCamera
 ]
 usedDynamicInputs [
  &SelectVectorFromArray
  &Multiply_Vector
  /Niagara/DynamicInputs/TypeConversions/MakeCustomPositionFromBool
  &FN_FrontEndVector3Modifier
 ]
 usedModules [
  &self:SystemSpawnScript.SystemScriptSource.SystemScriptGraph.NiagaraNodeAssignment_0.SetVariables_#1
  &self:SystemSpawnScript.SystemScriptSource.SystemScriptGraph.NiagaraNodeAssignment_1.SetVariables_#2
  &self:SystemSpawnScript.SystemScriptSource.SystemScriptGraph.NiagaraNodeAssignment_6.SetVariables_#3
  &self:SystemSpawnScript.SystemScriptSource.SystemScriptGraph.NiagaraNodeAssignment_4.SetVariables_#4
  &self:SystemSpawnScript.SystemScriptSource.SystemScriptGraph.NiagaraNodeAssignment_7.SetVariables_#5
  &self:SystemSpawnScript.SystemScriptSource.SystemScriptGraph.NiagaraNodeAssignment_5.SetVariables_#6
  &self:SystemSpawnScript.SystemScriptSource.SystemScriptGraph.NiagaraNodeAssignment_2.SetVariables_#7
  &SystemState
 ]
}
emitters {
 FireParticles {
  emitterUpdateScript [
   &EmitterState(
    Life Cycle Mode=::NewEnumerator1
    Loop Behavior=::NewEnumerator1
    Loop Delay=0.35
    Loop Duration=&Add_Float(A=$System.IntroLength, B=-0.15)
    UseLoopDelay=true
   )
   &SpawnBurst_Instantaneous(Spawn Count=25, Use Spawn Probability=false, UseLoopCountLimit=false)
   &SpawnRate(SpawnRate=25, Use Spawn Probability=true)
   &SpawnBurst_Instantaneous(Spawn Count=25, Spawn Group=1, Use Spawn Probability=false, UseLoopCountLimit=false) {module=SpawnBurst_Instantaneous001}
   &SpawnRate(SpawnGroup=1, SpawnRate=25, Use Spawn Probability=true) {module=SpawnRate001}
   &SpawnBurst_Instantaneous(Spawn Count=5, Spawn Group=2, Use Spawn Probability=false, UseLoopCountLimit=false) {module=SpawnBurst_Instantaneous002}
   &SpawnBurst_Instantaneous(Spawn Count=5, Spawn Group=3, Use Spawn Probability=false, UseLoopCountLimit=false) {module=SpawnBurst_Instantaneous003}
  ]
  particleSpawnScript [
   &InitializeParticle(
    Color=&MergeHelper_ColorSwitch(
     0=<rgba 1 1 1 1>
     1=<rgba 3 0.7919 0.3503 1>
     2=<rgba 1 1 1 1>
     3=&RandomRangeLinearColor(Minimum=<rgba 0.4896 0.4896 0.4896 1>, Maximum=<rgba 1 1 1 1>)
     4=&RandomRangeLinearColor(Minimum=<rgba 0.4896 0.4896 0.4896 1>, Maximum=<rgba 1 1 1 1>)
     Index=&SpawnGroup()
     Num=::NewEnumerator2
    )
    Lifetime=&MergeHelper_FloatSwitch(
     0=&RandomRangeFloat(Minimum=0.3, Maximum=0.8)
     1=&RandomRangeFloat(Minimum=0.3, Maximum=0.8)
     2=&RandomRangeFloat(Minimum=1.5, Maximum=2)
     3=&RandomRangeFloat(Minimum=0.8, Maximum=1.1)
     4=&RandomRangeFloat(Minimum=0.8, Maximum=1.1)
     Index=&SpawnGroup()
     Num=::NewEnumerator2
    )
    Sprite Rotation Mode=::NewEnumerator2
    Sprite Size=&Multiply_Vector2D_ByFloat(
     Float=&FN_FrontEndFloatModifier(Gameplay=1, Front End=$User.FrontEndScale)
     Vector2D=&MergeHelper_Vector2Switch(
      0=&Multiply_Vector2D_ByFloat(Float=&RandomRangeFloat(Minimum=1, Maximum=2), Vector2D=<5 5>)
      1=&RandomRangeVector2D(Minimum=<0.5 3>, Maximum=<0.7 1>)
      2=&Vector2DFromFloat(Value=&Multiply_Float(A=120, B=&Multiply_Float(A=0.15, B=0.25)))
      3=&Vector2DFromFloat(Value=&Multiply_Float(A=10, B=&Multiply_Float(A=0.4, B=0.6)))
      4=&Vector2DFromFloat(Value=&Multiply_Float(A=10, B=&Multiply_Float(A=0.4, B=0.6)))
      Index=&SpawnGroup()
      Num=::NewEnumerator2
     )
    )
    Sprite Size Mode=::NewEnumerator4
    UsePositionOffset=false
    Write Material Random=true
   )
   &ShapeLocation(MaskSpawnGroup=false, Override Seed=false, Sphere Radius=3, UseOwnerScale=true)
   &AddVelocity(
    Rotation Coordinate Space=::Simulation
    Velocity=&RandomRangeVector(Minimum=<-50 -50 -50>, Maximum=<50 50 50>)
    Velocity Speed Scale=&MergeHelper_FloatSwitch(
     0=&RandomRangeFloat(Minimum=0.75, Maximum=1.25)
     1=&RandomRangeFloat(Minimum=0.75, Maximum=1.25)
     2=&RandomRangeFloat(Minimum=0.5, Maximum=0.75)
     3=&RandomRangeFloat(Minimum=0.5, Maximum=0.75)
     4=&RandomRangeFloat(Minimum=0.2, Maximum=0.4)
     Index=&SpawnGroup()
     Num=::NewEnumerator2
    )
   ) {module=AddVelocity001}
   &DynamicMaterialParameters(
    Index 0 Param 2=&MergeHelper_FloatSwitch(
     0=&RandomRangeFloat(Minimum=0, Maximum=4)
     1=&RandomRangeFloat(Minimum=0, Maximum=4)
     2=0
     3=0
     4=0
     Index=&SpawnGroup()
     Num=::NewEnumerator3
    )
    Index 0 Param 4=2.5
    Module.Index0Float0Write=true
    Module.Index0Float1Write=true
    Module.Index0Float2Write=true
    Module.Index0Float3Write=true
   ) {module=DynamicMaterialParameters001}
   /SetVariables(Particles.VisibilityTag=&SpawnGroup()) {isSetParameters=true, module=SetVariables_#8}
   /SetVariables(Particles.AORandFloat:NiagaraFloat=&UniformAOrBFloat(A=0, Distribution Weight=0.3, Evaluation Type=::NewEnumerator0, Recalculate Random Each Loop=true, Override Seed=false)) {isSetParameters=true, module=SetVariables_#9}
   /SetVariables() {isSetParameters=true}
  ]
  particleUpdateScript [
   &ParticleState()
   &SpriteRotationRate(Rotation Rate=&MergeHelper_FloatSwitch(
     0=&RandomRangeFloat(Minimum=-720, Maximum=720)
     1=&RandomRangeFloat(Minimum=-720, Maximum=720)
     2=&RandomRangeFloat(Minimum=-300, Maximum=300)
     3=&RandomRangeFloat(Minimum=-300, Maximum=300)
     4=&RandomRangeFloat(Minimum=-300, Maximum=300)
     Index=&SpawnGroup()
     Num=::NewEnumerator2
    ))
   &SubUVAnimation(
    End Frame Range Override=11
    Evaluation Type=::NewEnumerator0
    "Sprite Renderer "=DataInterface() {propertyValues {SpriteRenderer=&self:FireParticles_0.NiagaraSpriteRendererProperties_1}}
    Start Frame Range Override=1
    SubUV Animation Mode:ENiagara_SubUVLookupModeV2=::NewEnumerator2
    UseEndFrame=true
    UseStartFrame=true
   ) {module=SubUVAnimation001}
   &SubUVAnimation(
    End Frame Range Override=12
    Evaluation Type=::NewEnumerator0
    "Sprite Renderer "=DataInterface() {propertyValues {SpriteRenderer=&self:FireParticles_0.NiagaraSpriteRendererProperties_14}}
    Start Frame Range Override=8
    SubUV Animation Mode:ENiagara_SubUVLookupModeV2=::NewEnumerator2
    UseEndFrame=false
    UseStartFrame=false
   ) {module=SubUVAnimation003}
   &SubUVAnimation(
    End Frame Range Override=12
    Evaluation Type=::NewEnumerator0
    "Sprite Renderer "=DataInterface() {propertyValues {SpriteRenderer=&self:FireParticles_0.NiagaraSpriteRendererProperties_12}}
    Start Frame Range Override=8
    SubUV Animation Mode:ENiagara_SubUVLookupModeV2=::NewEnumerator2
    UseEndFrame=true
    UseStartFrame=true
   ) {module=SubUVAnimation002}
   &OffsetPosition(Position Offset=$System.BurstPosition)
   &ScaleSpriteSize(
    Scale Sprite Size Mode=::NewEnumerator0
    Uniform Scale Factor=&MergeHelper_FloatSwitch(
     0=&FloatFromCurve(FloatCurve=DataInterface() {propertyValues {Curve {keys [<at 0 0>,<at 0.1 0.5>,<at 1 0>]}}})
     1=&FloatFromCurve(
      FloatCurve=DataInterface() {propertyValues {Curve {keys [
          <at 0 0 RCIM_Cubic>
          {interpMode=RCIM_Cubic, leaveTangent=-3.0296, tangentMode=RCTM_Break, time=0.2, value=1.2}
          {arriveTangent=-2.8894, interpMode=RCIM_Cubic, leaveTangent=-2.8894, tangentMode=RCTM_User, time=1, value=0.1}
         ]}}}
      Scale Curve=1.6
     )
     2=&FloatFromCurve(
      FloatCurve=DataInterface() {propertyValues {Curve {keys [
          {arriveTangent=1.3427, interpMode=RCIM_Cubic, leaveTangent=1.3427, tangentMode=RCTM_User, time=0, value=0}
          {interpMode=RCIM_Cubic, tangentMode=RCTM_User, time=0.2, value=0.3}
          {tangentMode=RCTM_User, time=1, value=0.1}
         ]}}}
      Scale Curve=12
     )
     3=&FloatFromCurve(
      FloatCurve=DataInterface() {propertyValues {Curve {keys [
          {arriveTangent=1.8745, interpMode=RCIM_Cubic, leaveTangent=1.8745, tangentMode=RCTM_User, time=0, value=0.4734}
          {tangentMode=RCTM_User, time=1, value=1}
         ]}}}
      Scale Curve=6
     )
     4=&FloatFromCurve(
      FloatCurve=DataInterface() {propertyValues {Curve {keys [
          {arriveTangent=1.8745, interpMode=RCIM_Cubic, leaveTangent=1.8745, tangentMode=RCTM_User, time=0, value=0.4734}
          {tangentMode=RCTM_User, time=1, value=1}
         ]}}}
      Scale Curve=6
     )
     Index=&SpawnGroup()
     Num=::NewEnumerator2
    )
   ) {module=ScaleSpriteSize001}
   &ScaleColor(
    Scale Alpha=&MergeHelper_FloatSwitch(
     0=&FloatFromCurve(
      FloatCurve=DataInterface() {propertyValues {Curve {
         defaultValue=1
         keys [
          {interpMode=RCIM_Cubic, tangentMode=RCTM_User, time=0, value=0}
          <at 0.1 1>
          {arriveTangent=-1.1111, interpMode=RCIM_Cubic, leaveTangent=-1.1111, time=0.9, value=1}
          {tangentMode=RCTM_User, time=1, value=0}
         ]
        }}}
      Scale Curve=2
     )
     1=&FloatFromCurve(
      FloatCurve=DataInterface() {propertyValues {Curve {
         defaultValue=1
         keys [
          {interpMode=RCIM_Cubic, tangentMode=RCTM_User, time=0, value=0}
          <at 0.1 1>
          {arriveTangent=-1.1111, interpMode=RCIM_Cubic, leaveTangent=-1.1111, time=0.9, value=1}
          {tangentMode=RCTM_User, time=1, value=0}
         ]
        }}}
      Scale Curve=2
     )
     2=&Multiply_Float(
      A=1
      B=&FloatFromCurve(
       FloatCurve=DataInterface() {propertyValues {Curve {keys [
           {arriveTangent=3.6955, interpMode=RCIM_Cubic, leaveTangent=3.6955, tangentMode=RCTM_User, time=0, value=0}
           {interpMode=RCIM_Cubic, tangentMode=RCTM_User, time=0.5, value=1}
           {tangentMode=RCTM_User, time=1, value=0.1}
          ]}}}
       Scale Curve=&Multiply_Float(
        A=&FN_FrontEndFloatModifier(Gameplay=0.6, Front End=0.4)
        B=&Lerp_Float(A=0.2, B=0.7, Alpha=$Particles.AORandFloat)
       )
      )
     )
     3=&Multiply_Float(
      A=2
      B=&FloatFromCurve(
       FloatCurve=DataInterface() {propertyValues {Curve {keys [<at 0 1 RCIM_Cubic>, {interpMode=RCIM_Cubic, tangentMode=RCTM_User, time=0.1942, value=1}, {tangentMode=RCTM_User, time=1, value=0}]}}}
       Scale Curve=0.7
      )
     )
     4=&Multiply_Float(
      A=2
      B=&FloatFromCurve(
       FloatCurve=DataInterface() {propertyValues {Curve {keys [<at 0 1 RCIM_Cubic>, {interpMode=RCIM_Cubic, tangentMode=RCTM_User, time=0.1942, value=1}, {tangentMode=RCTM_User, time=1, value=0}]}}}
       Scale Curve=0.7
      )
     )
     Index=&SpawnGroup()
     Num=::NewEnumerator2
    )
    Scale RGB=&MergeHelper_VectorSwitch(
     0=&MakeVectorFromLinearColorRGB(Color=&ColorFromCurve(ColorCurve=DataInterface() {propertyValues {
         AlphaCurve {keys []}
         BlueCurve {keys [<at 0 0 RCIM_Cubic>,<at 1 0 RCIM_Cubic>]}
         GreenCurve {keys [<at 0 0.2885 RCIM_Cubic>,<at 1 0.0015 RCIM_Cubic>]}
         RedCurve {keys [<at 0 2 RCIM_Cubic>,<at 1 0.0262 RCIM_Cubic>]}
        }}))
     1=<1 1 1>
     2=&Multiply_VectorByFloat(
      Vector=&MakeVectorFromLinearColorRGB(Color=&LerpLinearColors(
        StartColor=&ColorFromCurve(ColorCurve=DataInterface() {propertyValues {
           AlphaCurve {keys []}
           BlueCurve {keys [<at 0.6023 0.0007 RCIM_Cubic>]}
           GreenCurve {keys [<at 0.6023 0.3353 RCIM_Cubic>]}
           RedCurve {keys [<at 0.6023 4 RCIM_Cubic>]}
          }})
        EndColor=<rgba 0.0938 0.0138 0.003 1>
        LerpFactor=$Particles.AORandFloat
       ))
      Float=1
     )
     3=&Multiply_VectorByFloat(
      Vector=&MakeVectorFromLinearColorRGB(Color=&ColorFromCurve(ColorCurve=DataInterface() {propertyValues {
          AlphaCurve {keys []}
          BlueCurve {keys [<at 0.0268 0.0698>,<at 0.1891 0.0006>,<at 0.4587 0>,<at 0.7901 0.0346 RCIM_Cubic>]}
          GreenCurve {keys [<at 0.0268 0.3962>,<at 0.1891 0.0264>,<at 0.4587 0.0243>,<at 0.7901 0.2452 RCIM_Cubic>]}
          RedCurve {keys [<at 0.0268 2>,<at 0.1891 0.1146>,<at 0.4587 0.1146>,<at 0.7901 2 RCIM_Cubic>]}
         }}))
      Float=&FloatFromCurve(FloatCurve=DataInterface() {propertyValues {Curve {keys [<at 0.1 2>,<at 0.519 1>,<at 1 0>]}}})
     )
     4=&Multiply_VectorByFloat(
      Vector=&MakeVectorFromLinearColorRGB(Color=&ColorFromCurve(ColorCurve=DataInterface() {propertyValues {
          AlphaCurve {keys []}
          BlueCurve {keys [<at 0.0268 0.0698>,<at 0.1891 0.0006>,<at 0.4587 0>,<at 0.7901 0.0346 RCIM_Cubic>]}
          GreenCurve {keys [<at 0.0268 0.3962>,<at 0.1891 0.0264>,<at 0.4587 0.0243>,<at 0.7901 0.2452 RCIM_Cubic>]}
          RedCurve {keys [<at 0.0268 2>,<at 0.1891 0.1146>,<at 0.4587 0.1146>,<at 0.7901 2 RCIM_Cubic>]}
         }}))
      Float=&FloatFromCurve(FloatCurve=DataInterface() {propertyValues {Curve {keys [<at 0.1 2>,<at 0.519 1>,<at 1 0>]}}})
     )
     Index=&SpawnGroup()
     Num=::NewEnumerator2
    )
    ScaleA=true
    ScaleRGB=true
   )
   &CurlNoiseForce(
    Bias Noise=false
    Noise Frequency=25
    Noise Strength=&MergeHelper_FloatSwitch(
     0=500
     1=500
     2=&FloatFromCurve(FloatCurve=DataInterface() {propertyValues {Curve {keys [<at 0 0 RCIM_Cubic>,<at 1 1>]}}}, Scale Curve=1500)
     3=&FloatFromCurve(FloatCurve=DataInterface() {propertyValues {Curve {keys [<at 0 0 RCIM_Cubic>,<at 1 1>]}}}, Scale Curve=1500)
     Index=&SpawnGroup()
     Num=::NewEnumerator2
    )
    Pan Noise=true
    Pan Noise Field=<0 0 1>
    Random Seed:NiagaraInt32=&MergeHelper_IntSwitch(0=11, 1=12, 2=13, 3=14, 4=15, 5=?, Index=&SpawnGroup(), Num=::NewEnumerator3)
   )
   &ScaleVelocity(Velocity Scale=&VectorFromFloat(Value=&Multiply_Float(
      A=&FN_FrontEndFloatModifier(Gameplay=1, Front End=$User.FrontEndScale)
      B=&FloatFromCurve(
       FloatCurve=DataInterface() {propertyValues {Curve {keys [
           {arriveTangent=-0.1413, interpMode=RCIM_Cubic, time=0, value=0.8}
           {arriveTangent=-0.1274, interpMode=RCIM_Cubic, leaveTangent=-0.1274, tangentMode=RCTM_User, time=0.4, value=0.2}
           <at 0.9984 0 RCIM_Cubic>
          ]}}}
       Scale Curve=&MergeHelper_FloatSwitch(0=2.5, 1=2.5, 2=2, 3=2, Index=&SpawnGroup(), Num=::NewEnumerator2)
      )
     )))
   &Drag(
    Drag=&MergeHelper_FloatSwitch(0=2, 1=2, 2=4, 3=4, Index=&SpawnGroup(), Num=::NewEnumerator2)
    Use Linear Drag=true
    Use Rotational Drag=true
   )
   &GravityForce(Gravity=&RandomRangeVector(Minimum=<0 0 25>, Maximum=<0 0 50>))
   &SolveForcesAndVelocity(Clamp Velocity=false, Limit Acceleration=false)
   &ScaleSpriteSizeBySpeed(Max Scale Factor=&MergeHelper_Vector2Switch(0=<1 1>, 1=<1 3>, 2=<1 1>, 3=<1 1>, 4=?, Index=&SpawnGroup(), Num=::NewEnumerator2))
   &FN_Sprite_Attributes(
    Color=&MakeLinearColorFromVectorAndFloat(
     "Vector (RGB)"=&MakeVectorFromLinearColorRGB(Color=<rgba 2 0.2885 0 1>)
     "Float (Alpha)"=&Multiply_Float(
      A=&FN_FrontEndFloatModifier(Gameplay=1, Front End=0.2)
      B=&FloatFromCurve(
       FloatCurve=DataInterface() {propertyValues {Curve {keys [
           {interpMode=RCIM_Cubic, tangentMode=RCTM_User, time=0, value=1}
           {arriveTangent=-2.7437, interpMode=RCIM_Cubic, leaveTangent=-2.7437, tangentMode=RCTM_User, time=1, value=0}
          ]}}}
       Scale Curve=0.25
      )
     )
    )
    Create Color Attribute=true
    Create SpriteSize Attribute=true
    Module Output:ENiagara_AttributeSamplingApplyOutput=::NewEnumerator1
    SpriteSize=&Multiply_Vector2D_ByFloat(
     Float=&MakeFloatFromVector2d(Vector 2D=&Vector2DFromFloat(Value=&FloatFromCurve(FloatCurve=DataInterface() {propertyValues {Curve {keys [<at 0 0>,<at 0.1 0.5>,<at 1 0>]}}})))
     Vector2D=&Multiply_Vector2D_ByFloat(Float=5, Vector2D=$Particles.SpriteSize)
    )
   ) {module=FN_Sprite_Attributes002}
   &FN_Sprite_Attributes(
    Color=&MakeLinearColorFromVectorAndFloat(
     "Vector (RGB)"=&MakeVectorFromLinearColorRGB(Color=<rgba 1 0.0833 0 0.2>)
     "Float (Alpha)"=&FloatFromCurve(
      FloatCurve=DataInterface() {propertyValues {Curve {keys [
          {interpMode=RCIM_Cubic, tangentMode=RCTM_User, time=0, value=1}
          {arriveTangent=-2.7437, interpMode=RCIM_Cubic, leaveTangent=-2.7437, tangentMode=RCTM_User, time=1, value=0}
         ]}}}
      Scale Curve=&Multiply_Float(A=0.2, B=1)
     )
    )
    Create Color Attribute=true
    Create Dynamic Material Parameter Attributes=true
    Create SpriteSize Attribute=true
    Index 0 Param 1=2
    Module Output:ENiagara_AttributeSamplingApplyOutput=::NewEnumerator1
    Module.Index0Float0Write=true
    Module.Index0Float1Write=true
    Module.Index0Float2Write=true
    Module.Index0Float3Write=true
    SpriteSize=&Multiply_Vector2D_ByFloat(Float=5, Vector2D=&Multiply_Vector2D(A=$Particles.SpriteSize, B=<1 0.8>))
   )
   &FN_Sprite_Attributes(
    Color=&MakeLinearColorFromVectorAndFloat(
     "Vector (RGB)"=&MakeVectorFromLinearColorRGB(Color=&LerpLinearColors(StartColor=<rgba 1 0.2 0 0.5>, EndColor=<rgba 0.1719 0.0344 0 1>, LerpFactor=$Particles.AORandFloat))
     "Float (Alpha)"=&Multiply_Float(
      A=&FloatFromCurve(
       FloatCurve=DataInterface() {propertyValues {Curve {keys [
           {interpMode=RCIM_Cubic, tangentMode=RCTM_User, time=0, value=1}
           {arriveTangent=-2.7437, interpMode=RCIM_Cubic, leaveTangent=-2.7437, tangentMode=RCTM_User, time=1, value=0}
          ]}}}
       Scale Curve=0.14
      )
      B=&FN_FrontEndFloatModifier(Gameplay=0.5, Front End=0.1)
     )
    )
    Create Color Attribute=true
    Create SpriteSize Attribute=true
    Module Output:ENiagara_AttributeSamplingApplyOutput=::NewEnumerator1
    SpriteSize=&Multiply_Vector2D(A=$Particles.SpriteSize, B=<2 2>)
   ) {module=FN_Sprite_Attributes003}
   &FN_Sprite_Attributes(
    Color=&MakeLinearColorFromVectorAndFloat(
     "Vector (RGB)"=&MakeVectorFromLinearColorRGB(Color=<rgba 10 0.6212 0.1562 0.5>)
     "Float (Alpha)"=&FloatFromCurve(
      FloatCurve=DataInterface() {propertyValues {Curve {keys [
          {interpMode=RCIM_Cubic, tangentMode=RCTM_User, time=0, value=1}
          {arriveTangent=-2.7437, interpMode=RCIM_Cubic, leaveTangent=-2.7437, tangentMode=RCTM_User, time=1, value=0}
         ]}}}
      Scale Curve=0.7
     )
    )
    Create Color Attribute=true
    Create Dynamic Material Parameter Attributes=true
    Create SpriteSize Attribute=true
    Index 0 Param 1=3
    Index 0 Param 2=1
    Index 0 Param 3=1
    Index 0 Param 4=1
    Module Output:ENiagara_AttributeSamplingApplyOutput=::NewEnumerator1
    Module.Index0Float0Write=true
    Module.Index0Float1Write=true
    Module.Index0Float2Write=true
    Module.Index0Float3Write=true
    SpriteSize=&Multiply_Vector2D_ByFloat(Float=5, Vector2D=$Particles.SpriteSize)
   ) {module=FN_Sprite_Attributes004}
   &FN_Sprite_Attributes(
    Color=&MakeLinearColorFromVectorAndFloat(
     "Vector (RGB)"=&MakeVectorFromLinearColorRGB(Color=<rgba 1 0.2 0 0.5>)
     "Float (Alpha)"=&Multiply_Float(
      A=&FloatFromCurve(
       FloatCurve=DataInterface() {propertyValues {Curve {keys [
           {interpMode=RCIM_Cubic, tangentMode=RCTM_User, time=0, value=1}
           {arriveTangent=-2.7437, interpMode=RCIM_Cubic, leaveTangent=-2.7437, tangentMode=RCTM_User, time=1, value=0}
          ]}}}
       Scale Curve=0.4
      )
      B=&FN_FrontEndFloatModifier(Gameplay=0.3, Front End=0.1)
     )
    )
    Create Color Attribute=true
    Create SpriteSize Attribute=true
    Module Output:ENiagara_AttributeSamplingApplyOutput=::NewEnumerator1
    SpriteSize=&Multiply_Vector2D_ByFloat(Float=3, Vector2D=$Particles.SpriteSize)
   ) {module=FN_Sprite_Attributes005}
  ]
  rendererClasses [&NiagaraSpriteRendererProperties]
  renderers [
   {rendererClass=&NiagaraSpriteRendererProperties}
   {rendererClass=&NiagaraSpriteRendererProperties, rendererIndex=1}
   {rendererClass=&NiagaraSpriteRendererProperties, rendererIndex=2}
   {rendererClass=&NiagaraSpriteRendererProperties, rendererIndex=3}
   {rendererClass=&NiagaraSpriteRendererProperties, rendererIndex=4}
   {rendererClass=&NiagaraSpriteRendererProperties, rendererIndex=5}
   {rendererClass=&NiagaraSpriteRendererProperties, rendererIndex=6}
   {rendererClass=&NiagaraSpriteRendererProperties, rendererIndex=7}
   {rendererClass=&NiagaraSpriteRendererProperties, rendererIndex=8}
  ]
 }
 Heart {
  emitterSpawnScript [
   &InitializeAudioPlayer(Audio Player=DataInterface() {propertyValues {
      Attenuation=null
      Concurrency=null
      MaxPlaysPerTick=10
      SoundToPlay=/CosmeticCompanions/Assets/Audio_Shared/Interaction/Petting/SC_Mimosa_Interaction_React_Default
      bAllowLoopingOneShotSounds=false
      bLimitPlaysPerTick=true
      bOnlyActiveDuringGameplay=false
      bStopWhenComponentIsDestroyed=true
      ConfigurationUserParameter {parameter {name=None, typeDefHandle {registeredTypeIndex=1}, varData []}}
      ParameterNames []
     }})
   /Niagara/Modules/Audio/PlayAudio() {module=PlayAudio001}
  ]
  emitterUpdateScript [
   &EmitterState()
   &SpawnBurst_Instantaneous(Use Spawn Probability=false, UseLoopCountLimit=false)
  ]
  particleSpawnScript [
   /SetVariables(Particles.VisibilityTag=$Emitter.SpawnGroup) {isSetParameters=true}
   &InitializeParticle(
    Lifetime=&Add_Float(A=$System.IntroLength, B=0.25)
    Position Offset=<0 0 -1>
    Ribbon Width=10
    Ribbon Width Mode=::NewEnumerator1
    Sprite Rotation Mode=::NewEnumerator0
    Sprite Size Mode=::NewEnumerator3
    Uniform Sprite Size=&Multiply_Float(A=20, B=&FN_FrontEndFloatModifier(Gameplay=1, Front End=$User.FrontEndScale))
    UsePositionOffset=true
    Write Material Random=true
   )
   &CameraOffset(Camera Offset Amount=5)
  ]
  particleUpdateScript [
   &ParticleState()
   &ScaleSpriteSize(
    Scale Sprite Size Mode=::NewEnumerator0
    Uniform Scale Factor=&Add_Float(
     A=0
     B=&FloatFromCurve(FloatCurve=DataInterface() {propertyValues {Curve {
         defaultValue=0
         keys [
          <at 0.3 0.3 RCIM_Cubic>
          {arriveTangent=1.25, time=0.45, value=1}
          {arriveTangent=1.1016, interpMode=RCIM_Cubic, leaveTangent=1.1016, tangentMode=RCTM_User, time=0.9, value=1}
          {arriveTangent=9.7177, interpMode=RCIM_Cubic, leaveTangent=9.7177, tangentMode=RCTM_User, time=1, value=1.5}
         ]
        }}})
    )
   ) {module=ScaleSpriteSize002}
   &ScaleColor(
    Scale Alpha=&FloatFromCurve(FloatCurve=DataInterface() {propertyValues {Curve {
        defaultValue=0
        keys [
         {arriveTangent=9.0725, interpMode=RCIM_Cubic, leaveTangent=9.0725, tangentMode=RCTM_User, time=0.3, value=0}
         {arriveTangent=-0.2441, time=0.45, value=0.9847}
         {arriveTangent=-1.6411, leaveTangent=-1.6411, time=0.85, value=1}
         {arriveTangent=-5.7923, interpMode=RCIM_Cubic, leaveTangent=-5.7923, tangentMode=RCTM_User, time=1, value=0}
        ]
       }}})
    ScaleA=true
    ScaleRGB=true
   )
   &OffsetPosition(Position Offset=$System.BurstPosition)
   /SetVariables(
    Particles.Glow Color=&MultiplyLinearColor(
     Linear Color=$Particles.Color
     Scale Factor=&MakeVector4FromVectorAndFloat(
      "Vector (RGB)"=&MakeVectorFromLinearColorRGB(Color=<rgba 1 0.1356 0 1>)
      "Float (Alpha)"=&Multiply_Float(
       A=&Waveform(
        Waveform Count=::NewEnumerator1
        "Waveform [1]"=::NewEnumerator0
        "[1] Frequency"=3
        "[1] Amplitude Min/Max"=<0.3 1>
        "[2] Average"=false
        "Waveform [2]"=::NewEnumerator8
        "[2] Input Position"=0
        "[2] Phase"=$Emitter.Age
        "[2] Phase Scale"=1
        "[2] Compound Frequency"=2
        "[2] Frequency"=25
        "[2] Amplitude Min/Max"=<0.3 1>
        "[2] Scale"=1
        "[2] Bias"=0
        "[2] Exponent"=1
        "[2] Pulse Duration"=0.1
        "[2] Clip Triangle Min/Max"=<0 0>
        "[2] Random Seed"=0
       )
       B=&FN_FrontEndFloatModifier(Gameplay=1, Front End=0.35)
      )
     )
    )
    Particles.Glow Size=&Multiply_Vector2D(A=$Particles.SpriteSize, B=<3 3>)
   ) {isSetParameters=true}
   &SubUVAnimation(
    End Frame Range Override=8
    Play Rate=1.5
    Random Start Frame=true
    "Sprite Renderer "=DataInterface() {propertyValues {SpriteRenderer=&self:Heart_0.NiagaraSpriteRendererProperties_2}}
    SubUV Animation Mode:ENiagara_SubUVLookupModeV2=::NewEnumerator3
    UseEndFrame=true
    UsePlayRateScale=false
    UseStartFrame=true
   )
  ]
  rendererClasses [&NiagaraSpriteRendererProperties]
  renderers [
   {rendererClass=&NiagaraSpriteRendererProperties}
   {rendererClass=&NiagaraSpriteRendererProperties, rendererIndex=1}
  ]
 }
 HeartFire {
  emitterUpdateScript [
   &EmitterState()
   &SpawnBurst_Instantaneous(Spawn Count=5, Use Spawn Probability=false, UseLoopCountLimit=false)
   &SpawnBurst_Instantaneous(Spawn Count=5, Spawn Group=1, Use Spawn Probability=false, UseLoopCountLimit=false) {module=SpawnBurst_Instantaneous001}
   &SpawnBurst_Instantaneous(Spawn Count=5, Spawn Group=2, Use Spawn Probability=false, UseLoopCountLimit=false) {module=SpawnBurst_Instantaneous002}
  ]
  particleSpawnScript [
   /SetVariables(Particles.VisibilityTag=&SpawnGroup()) {isSetParameters=true, module=SetVariables_#10}
   &PartitionParticles(Number Of Partitions=5, Particles Per Partition=1, Partition Mode:ENiagara_PartitionMode_v2=::NewEnumerator1)
   &InitializeParticle(
    Lifetime=&Add_Float(A=$System.IntroLength, B=0.25)
    Position Mode=::NewEnumerator0
    Sprite Rotation Mode=::NewEnumerator0
    Sprite Size Mode=::NewEnumerator3
    Uniform Sprite Size=&Multiply_Float(
     A=&Multiply_Float(
      A=&MergeHelper_FloatSwitch(0=15, 1=17.5, 2=20, 3=?, Index=&SpawnGroup(), Num=::NewEnumerator1)
      B=&MergeHelper_FloatSwitch(0=1.5, 1=1.5, 2=1.5, 3=1.5, 4=2, Index=$Particles.PartitionParticles.Partition, Num=::NewEnumerator3)
     )
     B=&FN_FrontEndFloatModifier(Gameplay=1, Front End=$User.FrontEndScale)
    )
    Write Material Random=true
   )
   /SetVariables(Particles.PivotOffset:Vector2f=/Niagara/DynamicInputs/Add/Add_Vector2(
     A:Vector2f=<0.5 0.5>
     B:Vector2f=&Multiply_Vector2D(
      A=&SelectVector2DFromArray(
       Direct Array Index=$Particles.PartitionParticles.Partition
       Vector 2D Selection Array=DataInterface() {propertyValues {FloatData [<0.3318 0.1896>, <-0.3318 0.1896>, <0.1066 -0.218>, <-0.1066 -0.218>, <0 -0.1327>]}}
      )
      B=<0.55 0.55>
     )
    )) {isSetParameters=true}
  ]
  particleUpdateScript [
   &ParticleState()
   &ScaleSpriteSize(Uniform Curve Sprite Scale=DataInterface() {propertyValues {Curve {
       defaultValue=0
       keys [
        {arriveTangent=1.0733, interpMode=RCIM_Cubic, leaveTangent=1.0733, tangentMode=RCTM_User, time=0.3, value=0.3}
        {arriveTangent=1.25, time=0.45, value=1}
        {arriveTangent=1.1016, interpMode=RCIM_Cubic, leaveTangent=1.1016, tangentMode=RCTM_User, time=0.9, value=1}
        {arriveTangent=9.7177, interpMode=RCIM_Cubic, leaveTangent=9.7177, tangentMode=RCTM_User, time=1, value=1.5}
       ]
      }}})
   &OffsetPosition(Position Offset=$System.BurstPosition)
   &Drag(Use Linear Drag=true, Use Rotational Drag=true)
   &SolveForcesAndVelocity(Clamp Velocity=false, Limit Acceleration=false)
   &ScaleColor(
    Scale Alpha=&FloatFromCurve(FloatCurve=DataInterface() {propertyValues {Curve {
        defaultValue=0
        keys [
         {arriveTangent=9.0725, interpMode=RCIM_Cubic, leaveTangent=9.0725, tangentMode=RCTM_User, time=0.3, value=0}
         {arriveTangent=-0.2441, time=0.45, value=0.9847}
         {arriveTangent=-1.6411, leaveTangent=-1.6411, time=0.8, value=1}
         {arriveTangent=-5.7923, interpMode=RCIM_Cubic, leaveTangent=-5.7923, tangentMode=RCTM_User, time=1, value=0}
        ]
       }}})
    Scale RGB=&Multiply_VectorByFloat(
     Vector=&MakeVectorFromLinearColorRGB(Color=&MergeHelper_ColorSwitch(
       0=<rgba 1 0.7517 0.007 1>
       1=<rgba 1 0.5457 0.0006 1>
       2=<rgba 1 0.1356 0 1>
       3=?
       Index=&SpawnGroup()
       Num=::NewEnumerator1
      ))
     Float=1
    )
    ScaleA=true
    ScaleRGB=true
   )
   &CameraOffset(Camera Offset Amount=10)
   &SubUVAnimation(
    Play Rate=0.5
    Random Start Frame=true
    "Sprite Renderer "=DataInterface() {propertyValues {SpriteRenderer=&self:HeartFire_0.NiagaraSpriteRendererProperties_1}}
    Start Frame Offset:NiagaraInt32=&RandomRangeInt(Minimum=0, Maximum=63)
    SubUV Animation Mode:ENiagara_SubUVLookupModeV2=::NewEnumerator3
    UseEndFrame=true
    UsePlayRateScale=false
    UseStartFrame=false
   )
   &SubUVAnimation(
    Play Rate=0.5
    Playback Scale Curve=DataInterface() {propertyValues {Curve {keys [
        {interpMode=RCIM_Cubic, tangentMode=RCTM_User, time=0.25, value=1}
        {arriveTangent=-1.2656, interpMode=RCIM_Cubic, leaveTangent=-1.2656, tangentMode=RCTM_User, time=1, value=0}
       ]}}}
    Random Start Frame=true
    "Sprite Renderer "=DataInterface() {propertyValues {SpriteRenderer=&self:HeartFire_0.NiagaraSpriteRendererProperties_0}}
    Start Frame Offset:NiagaraInt32=&RandomRangeInt(Minimum=0, Maximum=63)
    SubUV Animation Mode:ENiagara_SubUVLookupModeV2=::NewEnumerator3
    UseEndFrame=true
    UsePlayRateScale=false
    UseStartFrame=false
   ) {module=SubUVAnimation001}
   &SubUVAnimation(
    Play Rate=0.5
    Playback Scale Curve=DataInterface() {propertyValues {Curve {keys [
        {interpMode=RCIM_Cubic, tangentMode=RCTM_User, time=0.25, value=1}
        {arriveTangent=-1.2656, interpMode=RCIM_Cubic, leaveTangent=-1.2656, tangentMode=RCTM_User, time=1, value=0}
       ]}}}
    Random Start Frame=true
    "Sprite Renderer "=DataInterface() {propertyValues {SpriteRenderer=&self:HeartFire_0.NiagaraSpriteRendererProperties_2}}
    Start Frame Offset:NiagaraInt32=&RandomRangeInt(Minimum=0, Maximum=63)
    SubUV Animation Mode:ENiagara_SubUVLookupModeV2=::NewEnumerator3
    UseEndFrame=true
    UsePlayRateScale=false
    UseStartFrame=false
   ) {module=SubUVAnimation002}
   &DynamicMaterialParameters(
    Index 0 Param 1=&FloatFromCurve(FloatCurve=DataInterface() {propertyValues {Curve {keys [
         {arriveTangent=-24.8334, interpMode=RCIM_Cubic, leaveTangent=-24.8334, tangentMode=RCTM_User, time=0.3, value=3}
         {arriveTangent=-3.75, time=0.45, value=0}
         {arriveTangent=3.6364, interpMode=RCIM_Cubic, leaveTangent=3.6364, time=0.93, value=0}
         <at 1 2 RCIM_Cubic>
        ]}}})
    Index 0 Param 2=&RandomRangeFloat(Minimum=0, Maximum=1)
    Index 0 Param 3=1
    Index 0 Param 4=1
    Module.Index0Float0Write=true
    Module.Index0Float1Write=true
    Module.Index0Float2Write=true
    Module.Index0Float3Write=true
    Module.Index1Float0Write=true
    Module.Index1Float1Write=true
    Module.Index1Float2Write=true
    Module.Index1Float3Write=true
    Write Parameter Index 1=::NewEnumerator1
   ) {module=DynamicMaterialParameters002}
  ]
  rendererClasses [&NiagaraSpriteRendererProperties]
  renderers [
   {rendererClass=&NiagaraSpriteRendererProperties}
   {rendererClass=&NiagaraSpriteRendererProperties, rendererIndex=1}
   {rendererClass=&NiagaraSpriteRendererProperties, rendererIndex=2}
  ]
 }
 Intro_FireParticles {
  emitterUpdateScript [
   &EmitterState(Life Cycle Mode=::NewEnumerator1, Loop Behavior=::NewEnumerator1, Loop Delay=0.05, Loop Duration=$System.IntroLength, UseLoopDelay=false)
   &SpawnRate(
    SpawnRate=&FloatFromCurve(
     FloatCurve=DataInterface() {propertyValues {Curve {keys [<at 0 1 RCIM_Cubic>,<at 0.35 0>]}}}
     CurveIndex=&Divide_Float(A=$Emitter.Age, B=$System.IntroLength)
     Scale Curve=50
    )
    Use Spawn Probability=false
   )
   &SpawnRate(
    SpawnGroup=1
    SpawnRate=&FloatFromCurve(
     FloatCurve=DataInterface() {propertyValues {Curve {keys [<at 0 1 RCIM_Cubic>,<at 0.35 0>]}}}
     CurveIndex=&Divide_Float(A=$Emitter.Age, B=$System.IntroLength)
     Scale Curve=50
    )
    Use Spawn Probability=false
   ) {module=SpawnRate001}
   &SpawnRate(
    SpawnGroup=2
    SpawnRate=&FloatFromCurve(
     FloatCurve=DataInterface() {propertyValues {Curve {keys [<at 0 1 RCIM_Cubic>,<at 0.35 0>]}}}
     CurveIndex=&Divide_Float(A=$Emitter.Age, B=$System.IntroLength)
     Scale Curve=75
    )
    Use Spawn Probability=false
   ) {module=SpawnRate002}
  ]
  particleSpawnScript [
   /SetVariables(Particles.VisibilityTag=&SpawnGroup()) {isSetParameters=true, module=SetVariables_#8}
   &InitializeParticle(
    Color=&MergeHelper_ColorSwitch(
     0=&LerpLinearColors(
      StartColor=&RandomRangeLinearColor(Minimum=<rgba 1 0.0713 0 1>, Maximum=<rgba 1 0.4162 0 1>)
      EndColor=&RandomRangeLinearColor(Minimum=<rgba 0.1354 0.0097 0 2>, Maximum=<rgba 0.099 0.0263 0 2>)
      LerpFactor=&UniformAOrBFloat(A=0, Evaluation Type=::NewEnumerator0, Recalculate Random Each Loop=true, Override Seed=false)
     )
     1=<rgba 1 1 1 1>
     2=<rgba 3 0.7919 0.3503 1>
     3=?
     4=&RandomRangeLinearColor(Minimum=<rgba 0.4896 0.4896 0.4896 1>, Maximum=<rgba 1 1 1 1>)
     Index=&SpawnGroup()
     Num=::NewEnumerator1
    )
    Lifetime=&MergeHelper_FloatSwitch(
     0=&RandomRangeFloat(Minimum=0.7, Maximum=1.3)
     1=&RandomRangeFloat(Minimum=0.7, Maximum=1.3)
     2=&RandomRangeFloat(Minimum=0.5, Maximum=1)
     3=&RandomRangeFloat(Minimum=0.8, Maximum=1.1)
     4=&RandomRangeFloat(Minimum=0.8, Maximum=1.1)
     Index=&SpawnGroup()
     Num=::NewEnumerator1
    )
    Sprite Rotation Mode=::NewEnumerator2
    Sprite Size=&Multiply_Vector2D_ByFloat(
     Float=&FN_FrontEndFloatModifier(Gameplay=1, Front End=$User.FrontEndScale)
     Vector2D=&MergeHelper_Vector2Switch(
      0=&Multiply_Vector2D_ByFloat(Float=&RandomRangeFloat(Minimum=1, Maximum=2), Vector2D=<5 5>)
      1=&Multiply_Vector2D_ByFloat(Float=&RandomRangeFloat(Minimum=1, Maximum=2), Vector2D=<5 5>)
      2=&RandomRangeVector2D(Minimum=<0.3 2>, Maximum=<0.5 1>)
      3=&Vector2DFromFloat(Value=&Multiply_Float(A=120, B=&Multiply_Float(A=0.15, B=0.25)))
      4=&Vector2DFromFloat(Value=&Multiply_Float(A=10, B=&Multiply_Float(A=0.4, B=0.6)))
      Index=&SpawnGroup()
      Num=::NewEnumerator1
     )
    )
    Sprite Size Mode=::NewEnumerator4
    UsePositionOffset=false
    Write Material Random=true
   )
   &ShapeLocation(MaskSpawnGroup=false, Override Seed=false, Sphere Radius=3, UseOwnerScale=true)
   &AddVelocity(
    Velocity=&MergeHelper_VectorSwitch(
     0=&RandomRangeVector(Minimum=<-25 -25 50>, Maximum=<25 25 125>)
     1=&RandomRangeVector(Minimum=<-25 -25 50>, Maximum=<25 25 125>)
     2=&RandomRangeVector(Minimum=<-25 -25 50>, Maximum=<25 25 100>)
     3=?
     Index=&SpawnGroup()
     Num=::NewEnumerator1
    )
    Velocity Speed Scale=&RandomRangeFloat(Minimum=0.1, Maximum=0.3)
   ) {module=AddVelocity001}
   &DynamicMaterialParameters(
    Index 0 Param 2=&MergeHelper_FloatSwitch(0=2, 1=2, 2=0, 3=?, Index=&SpawnGroup(), Num=::NewEnumerator1)
    Index 0 Param 4=1
    Module.Param0WriteEnabled=true
    Module.Param1WriteEnabled=true
    Module.Param2WriteEnabled=false
    Module.Param3WriteEnabled=true
   ) {module=DynamicMaterialParameters001}
   &OffsetPosition(Position Offset=&Add_Vector(
     A=<0 0 0>
     B=&Add_Vector(
      A=&Lerp_Vector(
       A=$System.HeartIntroSpawnPosition
       B=$System.BurstPosition
       Alpha=&FloatFromCurve(
        FloatCurve=DataInterface() {propertyValues {Curve {keys [{arriveTangent=2.0116, interpMode=RCIM_Cubic, leaveTangent=2.0116, tangentMode=RCTM_User, time=0, value=0}, <at 0.3 1 RCIM_Cubic>]}}}
        CurveIndex=&Divide_Float(A=$Emitter.Age, B=$System.IntroLength)
       )
      )
      B=&MakeVector(
       X=&Waveform(
        Global Phase=&Multiply_Float(A=$Emitter.Age, B=2)
        "Waveform [1]"=::NewEnumerator0
        "[1] Frequency"=3
        "[1] Amplitude Min/Max"=<-1 1>
        "[1] Scale"=&FloatFromCurve(
         FloatCurve=DataInterface() {propertyValues {Curve {keys [{arriveTangent=1.7542, interpMode=RCIM_Cubic, time=0.2496, value=0.5534}, <at 0.35 0 RCIM_Cubic>]}}}
         CurveIndex=$Emitter.Age
         Scale Curve=&Multiply_Float(A=10, B=&FN_FrontEndFloatModifier(Gameplay=1, Front End=$User.FrontEndScale))
        )
       )
       Y=&Waveform(
        Global Phase=&Multiply_Float(A=$Emitter.Age, B=1.5)
        "Waveform [1]"=::NewEnumerator1
        "[1] Frequency"=3
        "[1] Amplitude Min/Max"=<-1 1>
        "[1] Scale"=&FloatFromCurve(
         FloatCurve=DataInterface() {propertyValues {Curve {keys [{arriveTangent=1.7542, interpMode=RCIM_Cubic, time=0.2496, value=0.5534}, <at 0.35 0 RCIM_Cubic>]}}}
         CurveIndex=$Emitter.Age
         Scale Curve=&Multiply_Float(A=10, B=&FN_FrontEndFloatModifier(Gameplay=1, Front End=$User.FrontEndScale))
        )
       )
       Z=0
      )
     )
    ))
   /SetVariables() {isSetParameters=true}
  ]
  particleUpdateScript [
   &ParticleState()
   &CurlNoiseForce(
    Bias Noise=false
    Noise Frequency=25
    Noise Strength=&FloatFromCurve(
     FloatCurve=DataInterface() {propertyValues {Curve {keys [{arriveTangent=2.1957, interpMode=RCIM_Cubic, leaveTangent=2.1957, tangentMode=RCTM_User, time=0, value=0}, <at 1 1>]}}}
     Scale Curve=&MergeHelper_FloatSwitch(0=1864.4294, 1=1864.4294, 2=900, 3=?, Index=&SpawnGroup(), Num=::NewEnumerator1)
    )
    Pan Noise=true
    Pan Noise Field=<0 0 1>
    Random Seed:NiagaraInt32=&MergeHelper_IntSwitch(0=11, 1=12, 2=13, 3=14, 4=15, 5=?, Index=&SpawnGroup(), Num=::NewEnumerator3)
   )
   &ScaleSpriteSize(
    Scale Sprite Size Mode=::NewEnumerator0
    Uniform Scale Factor=&MergeHelper_FloatSwitch(
     0=&FloatFromCurve(
      FloatCurve=DataInterface() {propertyValues {Curve {
         defaultValue=1
         keys [{arriveTangent=-1.9355, interpMode=RCIM_Cubic, time=0, value=1}, {interpMode=RCIM_Cubic, tangentMode=RCTM_User, time=1, value=0}]
        }}}
      Scale Curve=0.6
     )
     1=&FloatFromCurve(FloatCurve=DataInterface() {propertyValues {Curve {keys [<at 0 0>,<at 0.1 0.5>,<at 1 0>]}}})
     2=&FloatFromCurve(
      FloatCurve=DataInterface() {propertyValues {Curve {keys [
          <at 0 0 RCIM_Cubic>
          {interpMode=RCIM_Cubic, leaveTangent=-3.0296, tangentMode=RCTM_Break, time=0.2, value=1.2}
          {arriveTangent=-2.8894, interpMode=RCIM_Cubic, leaveTangent=-2.8894, tangentMode=RCTM_User, time=1, value=0.1}
         ]}}}
      Scale Curve=1.6
     )
     3=&FloatFromCurve(
      FloatCurve=DataInterface() {propertyValues {Curve {keys [
          {arriveTangent=1.3427, interpMode=RCIM_Cubic, leaveTangent=1.3427, tangentMode=RCTM_User, time=0, value=0}
          {interpMode=RCIM_Cubic, tangentMode=RCTM_User, time=0.2, value=0.3}
          {tangentMode=RCTM_User, time=1, value=0.1}
         ]}}}
      Scale Curve=12
     )
     4=&FloatFromCurve(
      FloatCurve=DataInterface() {propertyValues {Curve {keys [
          {arriveTangent=1.8745, interpMode=RCIM_Cubic, leaveTangent=1.8745, tangentMode=RCTM_User, time=0, value=0.4734}
          {tangentMode=RCTM_User, time=1, value=1}
         ]}}}
      Scale Curve=6
     )
     Index=&SpawnGroup()
     Num=::NewEnumerator1
    )
   ) {module=ScaleSpriteSize001}
   &ScaleColor(
    Scale Alpha=&MergeHelper_FloatSwitch(
     0=&FloatFromCurve(
      FloatCurve=DataInterface() {propertyValues {Curve {
         defaultValue=0
         keys [
          {arriveTangent=-2.5249, interpMode=RCIM_Cubic, leaveTangent=-2.5249, tangentMode=RCTM_User, time=0, value=0.9847}
          {arriveTangent=-0.0405, interpMode=RCIM_Cubic, leaveTangent=-0.0405, tangentMode=RCTM_User, time=1, value=0}
         ]
        }}}
      Scale Curve=0.4
     )
     1=2
     2=1
     3=?
     Index=&SpawnGroup()
     Num=::NewEnumerator1
    )
    Scale RGB=&MergeHelper_VectorSwitch(
     0=&VectorFromFloat(Value=&Lerp_Float(
       A=0
       B=&RandomRangeFloat(Minimum=3, Maximum=10)
       Alpha=&MakeFloatFromVector(
        Vector=&VectorFromFloat(Value=&FloatFromCurve(FloatCurve=DataInterface() {propertyValues {Curve {
             defaultValue=1
             keys [
              {arriveTangent=2.9173, interpMode=RCIM_Cubic, leaveTangent=2.9173, tangentMode=RCTM_User, time=0, value=0.7799}
              {arriveTangent=-2.5249, interpMode=RCIM_Cubic, leaveTangent=-2.5249, tangentMode=RCTM_User, time=0.3924, value=0.933}
              {arriveTangent=-0.0405, interpMode=RCIM_Cubic, leaveTangent=-0.0405, tangentMode=RCTM_User, time=1, value=0}
             ]
            }}}))
        Channel=::NewEnumerator0
       )
      ))
     1=&MakeVectorFromLinearColorRGB(Color=&ColorFromCurve(ColorCurve=DataInterface() {propertyValues {
         AlphaCurve {keys []}
         BlueCurve {keys [<at 0 0 RCIM_Cubic>,<at 1 0 RCIM_Cubic>]}
         GreenCurve {keys [<at 0 0.2885 RCIM_Cubic>,<at 1 0.0015 RCIM_Cubic>]}
         RedCurve {keys [<at 0 2 RCIM_Cubic>,<at 1 0.0262 RCIM_Cubic>]}
        }}))
     2=<1 1 1>
     3=<0 0 0>
     4=<0 0 0>
     Index=&SpawnGroup()
     Num=::NewEnumerator3
    )
    ScaleA=true
    ScaleRGB=true
   )
   &SubUVAnimation(
    End Frame Range Override=12
    Evaluation Type=::NewEnumerator0
    "Sprite Renderer "=DataInterface() {propertyValues {SpriteRenderer=&self:Intro_FireParticles_0.NiagaraSpriteRendererProperties_4}}
    Start Frame Range Override=8
    SubUV Animation Mode:ENiagara_SubUVLookupModeV2=::NewEnumerator2
    UseEndFrame=false
    UseStartFrame=false
   )
   &SubUVAnimation(
    End Frame Range Override=11
    Evaluation Type=::NewEnumerator0
    "Sprite Renderer "=DataInterface() {propertyValues {SpriteRenderer=&self:Intro_FireParticles_0.NiagaraSpriteRendererProperties_6}}
    Start Frame Range Override=1
    SubUV Animation Mode:ENiagara_SubUVLookupModeV2=::NewEnumerator2
    UseEndFrame=true
    UseStartFrame=true
   ) {module=SubUVAnimation001}
   &ScaleVelocity(Velocity Scale=&VectorFromFloat(Value=&Multiply_Float(
      A=&FN_FrontEndFloatModifier(Gameplay=1, Front End=$User.FrontEndScale)
      B=&MakeFloatFromVector(
       Vector=&VectorFromFloat(Value=&FloatFromCurve(
         FloatCurve=DataInterface() {propertyValues {Curve {keys [
             {arriveTangent=-0.1413, interpMode=RCIM_Cubic, time=0, value=0.8}
             {arriveTangent=-0.1274, interpMode=RCIM_Cubic, leaveTangent=-0.1274, tangentMode=RCTM_User, time=0.4, value=0.2}
             <at 0.9984 0 RCIM_Cubic>
            ]}}}
         Scale Curve=&MergeHelper_FloatSwitch(0=3.6, 1=3.6, 2=3.6, 3=?, Index=&SpawnGroup(), Num=::NewEnumerator1)
        ))
       Channel=::NewEnumerator0
      )
     )))
   &Drag(
    Drag=&MergeHelper_FloatSwitch(0=5, 1=5, 2=6, 3=?, Index=&SpawnGroup(), Num=::NewEnumerator1)
    Use Linear Drag=true
    Use Rotational Drag=true
   )
   &SolveForcesAndVelocity(Clamp Velocity=false, Limit Acceleration=false)
   &ScaleSpriteSizeBySpeed(Max Scale Factor=&MergeHelper_Vector2Switch(0=<1 1>, 1=<1 1>, 2=<1 3>, 3=?, 4=?, Index=&SpawnGroup(), Num=::NewEnumerator1))
   &FN_Sprite_Attributes(
    Color=&Multiply_LinearColorByFloat(Linear Color=$Particles.Color, Float=0.304)
    Create Color Attribute=true
    Create Dynamic Material Parameter Attributes=true
    Create SpriteSize Attribute=true
    Index 0 Param 1=3
    Index 0 Param 2=1
    Index 0 Param 3=1
    Index 0 Param 4=1
    Module Output:ENiagara_AttributeSamplingApplyOutput=::NewEnumerator1
    Module.Index0Float0Write=true
    Module.Index0Float1Write=true
    Module.Index0Float2Write=true
    Module.Index0Float3Write=true
    SpriteSize=&Multiply_Vector2D_ByFloat(Float=2, Vector2D=$Particles.SpriteSize)
   ) {module=FN_Sprite_Attributes001}
   &FN_Sprite_Attributes(
    Color=&MakeLinearColorFromVectorAndFloat(
     "Vector (RGB)"=&MakeVectorFromLinearColorRGB(Color=<rgba 2 0.2885 0 1>)
     "Float (Alpha)"=&Multiply_Float(
      A=&FN_FrontEndFloatModifier(Gameplay=1, Front End=0.2)
      B=&FloatFromCurve(
       FloatCurve=DataInterface() {propertyValues {Curve {keys [
           {interpMode=RCIM_Cubic, tangentMode=RCTM_User, time=0, value=1}
           {arriveTangent=-2.7437, interpMode=RCIM_Cubic, leaveTangent=-2.7437, tangentMode=RCTM_User, time=1, value=0}
          ]}}}
       Scale Curve=0.25
      )
     )
    )
    Create Color Attribute=true
    Create SpriteSize Attribute=true
    Module Output:ENiagara_AttributeSamplingApplyOutput=::NewEnumerator1
    SpriteSize=&Multiply_Vector2D_ByFloat(
     Float=&MakeFloatFromVector2d(Vector 2D=&Vector2DFromFloat(Value=&FloatFromCurve(FloatCurve=DataInterface() {propertyValues {Curve {keys [<at 0 0>,<at 0.1 0.5>,<at 1 0>]}}})))
     Vector2D=&Multiply_Vector2D_ByFloat(Float=10, Vector2D=$Particles.SpriteSize)
    )
   ) {module=FN_Sprite_Attributes002}
   &FN_Sprite_Attributes(
    Color=&MakeLinearColorFromVectorAndFloat(
     "Vector (RGB)"=&MakeVectorFromLinearColorRGB(Color=<rgba 1 0.0833 0 0.2>)
     "Float (Alpha)"=&FloatFromCurve(
      FloatCurve=DataInterface() {propertyValues {Curve {keys [
          {interpMode=RCIM_Cubic, tangentMode=RCTM_User, time=0, value=1}
          {arriveTangent=-2.7437, interpMode=RCIM_Cubic, leaveTangent=-2.7437, tangentMode=RCTM_User, time=1, value=0}
         ]}}}
      Scale Curve=&Multiply_Float(A=0.2, B=1)
     )
    )
    Create Color Attribute=true
    Create Dynamic Material Parameter Attributes=true
    Create SpriteSize Attribute=true
    Index 0 Param 1=2
    Module Output:ENiagara_AttributeSamplingApplyOutput=::NewEnumerator1
    Module.Index0Float0Write=true
    Module.Index0Float1Write=true
    Module.Index0Float2Write=true
    Module.Index0Float3Write=true
    SpriteSize=&Multiply_Vector2D_ByFloat(Float=5, Vector2D=&Multiply_Vector2D(A=$Particles.SpriteSize, B=<1 0.8>))
   )
  ]
  rendererClasses [&NiagaraSpriteRendererProperties]
  renderers [
   {rendererClass=&NiagaraSpriteRendererProperties}
   {rendererClass=&NiagaraSpriteRendererProperties, rendererIndex=1}
   {rendererClass=&NiagaraSpriteRendererProperties, rendererIndex=2}
   {rendererClass=&NiagaraSpriteRendererProperties, rendererIndex=3}
   {rendererClass=&NiagaraSpriteRendererProperties, rendererIndex=4}
   {rendererClass=&NiagaraSpriteRendererProperties, rendererIndex=5}
  ]
 }
 Intro_FireTrail {
  emitterSpawnScript [/SetVariables(Emitter.RandomOffset:NiagaraFloat=&RandomRangeFloat(Minimum=0, Maximum=30)) {isSetParameters=true}]
  emitterUpdateScript [
   &EmitterState()
   &SpawnRate(
    SpawnGroup=1
    SpawnRate=&FloatFromCurve(
     FloatCurve=DataInterface() {propertyValues {Curve {keys [<at 0 1 RCIM_Constant>,<at 0.22 0 RCIM_Constant>]}}}
     CurveIndex=$Emitter.NormalizedLoopAge
     Scale Curve=100
    )
    Use Spawn Probability=false
   )
   /SetVariables(Emitter.Dissolve:NiagaraFloat=&FloatFromCurve(
     FloatCurve=DataInterface() {propertyValues {Curve {keys [
         {arriveTangent=1.0887, interpMode=RCIM_Cubic, leaveTangent=1.0887, tangentMode=RCTM_User, time=0.2, value=0}
         {arriveTangent=6.7886, tangentMode=RCTM_User, time=0.35, value=1}
        ]}}}
     CurveIndex=&Divide_Float(A=$Emitter.Age, B=$System.IntroLength)
    )) {isSetParameters=true}
  ]
  particleSpawnScript [
   &InitializeParticle(
    Lifetime=$System.IntroLength
    Ribbon Width=&Multiply_Float(A=10, B=&FN_FrontEndFloatModifier(Gameplay=1, Front End=$User.FrontEndScale))
    Ribbon Width Mode=::NewEnumerator1
    Sprite Rotation Mode=::NewEnumerator0
    UsePositionOffset=false
    Write Material Random=true
   )
   /SetVariables(Particles.SpawnAge:NiagaraFloat=&Divide_Float(A=$Emitter.Age, B=$System.IntroLength)) {isSetParameters=true}
   /SetVariables(Particles.Position=&Add_Vector(
     A=<0 0 0>
     B=&Add_Vector(
      A=&Lerp_Vector(
       A=$System.HeartIntroSpawnPosition
       B=$System.BurstPosition
       Alpha=&FloatFromCurve(
        FloatCurve=DataInterface() {propertyValues {Curve {keys [{arriveTangent=2.0116, interpMode=RCIM_Cubic, leaveTangent=2.0116, tangentMode=RCTM_User, time=0, value=0}, <at 0.3 1 RCIM_Cubic>]}}}
        CurveIndex=&Divide_Float(A=$Emitter.Age, B=$System.IntroLength)
       )
      )
      B=&MakeVector(
       X=&Waveform(
        Global Phase=&Multiply_Float(A=$Emitter.Age, B=2)
        "Waveform [1]"=::NewEnumerator0
        "[1] Frequency"=3
        "[1] Amplitude Min/Max"=<-1 1>
        "[1] Scale"=&FloatFromCurve(
         FloatCurve=DataInterface() {propertyValues {Curve {keys [{arriveTangent=1.7542, interpMode=RCIM_Cubic, time=0.2496, value=0.5534}, <at 0.35 0 RCIM_Cubic>]}}}
         CurveIndex=$Emitter.Age
         Scale Curve=&Multiply_Float(A=10, B=&FN_FrontEndFloatModifier(Gameplay=1, Front End=$User.FrontEndScale))
        )
       )
       Y=&Waveform(
        Global Phase=&Multiply_Float(A=$Emitter.Age, B=1.5)
        "Waveform [1]"=::NewEnumerator1
        "[1] Frequency"=3
        "[1] Amplitude Min/Max"=<-1 1>
        "[1] Scale"=&FloatFromCurve(
         FloatCurve=DataInterface() {propertyValues {Curve {keys [{arriveTangent=1.7542, interpMode=RCIM_Cubic, time=0.2496, value=0.5534}, <at 0.35 0 RCIM_Cubic>]}}}
         CurveIndex=$Emitter.Age
         Scale Curve=&Multiply_Float(A=10, B=&FN_FrontEndFloatModifier(Gameplay=1, Front End=$User.FrontEndScale))
        )
       )
       Z=0
      )
     )
    )) {isSetParameters=true}
   &DynamicMaterialParameters(Module.Index0Param0Write=true, Module.Index0Param1Write=true, Module.Index0Param2Write=true, Module.Index0Param3Write=true)
  ]
  particleUpdateScript [
   &ParticleState()
   &ScaleColor(
    Scale Alpha=&FloatFromCurve(FloatCurve=DataInterface() {propertyValues {Curve {
        defaultValue=1
        keys [
         {interpMode=RCIM_Cubic, tangentMode=RCTM_User, time=0, value=1}
         {arriveTangent=0.045, interpMode=RCIM_Cubic, leaveTangent=0.045, tangentMode=RCTM_User, time=0.35, value=0}
        ]
       }}})
    Scale RGB=&MakeVectorFromLinearColorRGB(Color=&ColorFromCurve(ColorCurve=DataInterface() {propertyValues {
        AlphaCurve {keys [<at 0 1>]}
        BlueCurve {keys [<at 0 0>,<at 0.3738 0.0224>,<at 1 0.006>]}
        GreenCurve {keys [<at 0 0.1651>,<at 0.3738 0.1777>,<at 1 0.0095>]}
        RedCurve {keys [<at 0 1>,<at 0.3738 0.6435>,<at 1 0.0203>]}
       }}))
    ScaleA=true
    ScaleRGB=true
   )
   &ScaleRibbonWidth(Ribbon Width Scale=&Multiply_Float(
     A=&FloatFromCurve(FloatCurve=DataInterface() {propertyValues {Curve {keys [<at 0 0.25>,<at 0.1 1>,<at 1 2>]}}})
     B=&FloatFromCurve(FloatCurve=DataInterface() {propertyValues {Curve {keys [<at 0.05 0>,<at 0.15 1>,<at 0.85 1>,<at 1 0>]}}}, CurveIndex=$Particles.SpawnAge)
    )) {module=ScaleRibbonWidth001}
   &SolveForcesAndVelocity(Clamp Velocity=false, Limit Acceleration=false)
   /SetVariables(Particles.Linear Color=&MakeLinearColorFromVectorAndFloat(
     "Vector (RGB)"=<1 1 1>
     "Float (Alpha)"=&FloatFromCurve(FloatCurve=DataInterface() {propertyValues {Curve {
         defaultValue=1
         keys [
          {interpMode=RCIM_Cubic, tangentMode=RCTM_User, time=0, value=1}
          {arriveTangent=0.045, interpMode=RCIM_Cubic, leaveTangent=0.045, tangentMode=RCTM_User, time=0.35, value=0}
         ]
        }}})
    )) {isSetParameters=true}
  ]
  rendererClasses [&NiagaraRibbonRendererProperties]
  renderers [
   {rendererClass=&NiagaraRibbonRendererProperties}
   {rendererClass=&NiagaraRibbonRendererProperties, rendererIndex=1}
  ]
 }
 Intro_Flame {
  emitterUpdateScript [
   &EmitterState()
   &SpawnBurst_Instantaneous(Use Spawn Probability=false, UseLoopCountLimit=false)
   &SpawnBurst_Instantaneous(Spawn Count=2, Spawn Group=1, Use Spawn Probability=false, UseLoopCountLimit=false) {module=SpawnBurst_Instantaneous001}
   &SpawnBurst_Instantaneous(Spawn Count=3, Spawn Group=2, Use Spawn Probability=false, UseLoopCountLimit=false) {module=SpawnBurst_Instantaneous002}
  ]
  particleSpawnScript [
   /SetVariables(Particles.VisibilityTag=&SpawnGroup()) {isSetParameters=true, module=SetVariables_#10}
   &InitializeParticle(
    Lifetime=$System.IntroLength
    Sprite Rotation Mode=::NewEnumerator0
    Sprite Size Mode=::NewEnumerator3
    Uniform Sprite Size=&Multiply_Float(
     A=&Multiply_Float(
      A=&MergeHelper_FloatSwitch(
       0=&RandomRangeFloat(Minimum=15, Maximum=20)
       1=&RandomRangeFloat(Minimum=20, Maximum=25)
       2=&RandomRangeFloat(Minimum=25, Maximum=30)
       3=?
       Index=&SpawnGroup()
       Num=::NewEnumerator1
      )
      B=0.5
     )
     B=&FN_FrontEndFloatModifier(Gameplay=1, Front End=$User.FrontEndScale)
    )
    UsePositionOffset=false
    Write Material Random=true
   )
  ]
  particleUpdateScript [
   &ParticleState()
   &OffsetPosition(Position Offset=&Add_Vector(
     A=<0 0 0>
     B=&Add_Vector(
      A=&Lerp_Vector(
       A=$System.HeartIntroSpawnPosition
       B=$System.BurstPosition
       Alpha=&FloatFromCurve(
        FloatCurve=DataInterface() {propertyValues {Curve {keys [{arriveTangent=2.0116, interpMode=RCIM_Cubic, leaveTangent=2.0116, tangentMode=RCTM_User, time=0, value=0}, <at 0.3 1 RCIM_Cubic>]}}}
        CurveIndex=&Divide_Float(A=$Emitter.Age, B=$System.IntroLength)
       )
      )
      B=&MakeVector(
       X=&Waveform(
        Global Phase=&Multiply_Float(A=$Emitter.Age, B=2)
        "Waveform [1]"=::NewEnumerator0
        "[1] Frequency"=3
        "[1] Amplitude Min/Max"=<-1 1>
        "[1] Scale"=&FloatFromCurve(
         FloatCurve=DataInterface() {propertyValues {Curve {keys [{arriveTangent=1.7542, interpMode=RCIM_Cubic, time=0.2496, value=0.5534}, <at 0.35 0 RCIM_Cubic>]}}}
         CurveIndex=$Emitter.Age
         Scale Curve=&Multiply_Float(A=10, B=&FN_FrontEndFloatModifier(Gameplay=1, Front End=$User.FrontEndScale))
        )
       )
       Y=&Waveform(
        Global Phase=&Multiply_Float(A=$Emitter.Age, B=1.5)
        "Waveform [1]"=::NewEnumerator1
        "[1] Frequency"=3
        "[1] Amplitude Min/Max"=<-1 1>
        "[1] Scale"=&FloatFromCurve(
         FloatCurve=DataInterface() {propertyValues {Curve {keys [{arriveTangent=1.7542, interpMode=RCIM_Cubic, time=0.2496, value=0.5534}, <at 0.35 0 RCIM_Cubic>]}}}
         CurveIndex=$Emitter.Age
         Scale Curve=&Multiply_Float(A=10, B=&FN_FrontEndFloatModifier(Gameplay=1, Front End=$User.FrontEndScale))
        )
       )
       Z=0
      )
     )
    ))
   &ScaleColor(
    Scale Alpha=&FloatFromCurve(FloatCurve=DataInterface() {propertyValues {Curve {
        defaultValue=0
        keys [
         {arriveTangent=9.0725, interpMode=RCIM_Cubic, leaveTangent=9.0725, tangentMode=RCTM_User, time=0.31, value=0}
         {arriveTangent=-0.2441, time=0.45, value=1}
         {arriveTangent=-1.9231, interpMode=RCIM_Cubic, leaveTangent=-1.9231, time=0.9, value=1}
         {arriveTangent=-7.9849, interpMode=RCIM_Cubic, leaveTangent=-7.9849, tangentMode=RCTM_User, time=0.97, value=0}
        ]
       }}})
    Scale RGB=&Multiply_VectorByFloat(
     Vector=&MakeVectorFromLinearColorRGB(Color=&MergeHelper_ColorSwitch(
       0=<rgba 1 0.7517 0.007 1>
       1=<rgba 1 0.5457 0.0006 1>
       2=<rgba 1 0.1356 0 1>
       3=?
       Index=&SpawnGroup()
       Num=::NewEnumerator1
      ))
     Float=1
    )
    ScaleA=false
    ScaleRGB=true
   )
   &SubUVAnimation(
    Play Rate=0.5
    Random Start Frame=true
    "Sprite Renderer "=DataInterface() {propertyValues {SpriteRenderer=&self:Intro_Flame_0.NiagaraSpriteRendererProperties_1}}
    Start Frame Offset:NiagaraInt32=&RandomRangeInt(Minimum=0, Maximum=63)
    SubUV Animation Mode:ENiagara_SubUVLookupModeV2=::NewEnumerator3
    UseEndFrame=true
    UsePlayRateScale=false
    UseStartFrame=false
   )
   &SubUVAnimation(
    Play Rate=0.5
    Playback Scale Curve=DataInterface() {propertyValues {Curve {keys [
        {interpMode=RCIM_Cubic, tangentMode=RCTM_User, time=0.25, value=1}
        {arriveTangent=-1.2656, interpMode=RCIM_Cubic, leaveTangent=-1.2656, tangentMode=RCTM_User, time=1, value=0}
       ]}}}
    Random Start Frame=true
    "Sprite Renderer "=DataInterface() {propertyValues {SpriteRenderer=&self:Intro_Flame_0.NiagaraSpriteRendererProperties_0}}
    Start Frame Offset:NiagaraInt32=&RandomRangeInt(Minimum=0, Maximum=63)
    SubUV Animation Mode:ENiagara_SubUVLookupModeV2=::NewEnumerator3
    UseEndFrame=true
    UsePlayRateScale=false
    UseStartFrame=false
   ) {module=SubUVAnimation001}
   &SubUVAnimation(
    Play Rate=0.5
    Playback Scale Curve=DataInterface() {propertyValues {Curve {keys [
        {interpMode=RCIM_Cubic, tangentMode=RCTM_User, time=0.25, value=1}
        {arriveTangent=-1.2656, interpMode=RCIM_Cubic, leaveTangent=-1.2656, tangentMode=RCTM_User, time=1, value=0}
       ]}}}
    Random Start Frame=true
    "Sprite Renderer "=DataInterface() {propertyValues {SpriteRenderer=&self:Intro_Flame_0.NiagaraSpriteRendererProperties_2}}
    Start Frame Offset:NiagaraInt32=&RandomRangeInt(Minimum=0, Maximum=63)
    SubUV Animation Mode:ENiagara_SubUVLookupModeV2=::NewEnumerator3
    UseEndFrame=true
    UsePlayRateScale=false
    UseStartFrame=false
   ) {module=SubUVAnimation002}
   &DynamicMaterialParameters(
    Index 0 Param 1=&FloatFromCurve(FloatCurve=DataInterface() {propertyValues {Curve {keys [
         <at 0 3 RCIM_Cubic>
         {arriveTangent=-3.75, time=0.05, value=0}
         {arriveTangent=2.1053, interpMode=RCIM_Cubic, leaveTangent=2.1053, time=0.95, value=0}
         <at 1 2 RCIM_Cubic>
        ]}}})
    Index 0 Param 2=&RandomRangeFloat(Minimum=0, Maximum=1)
    Index 0 Param 3=1
    Index 0 Param 4=1
    Module.Index0Float0Write=true
    Module.Index0Float1Write=true
    Module.Index0Float2Write=true
    Module.Index0Float3Write=true
   ) {module=DynamicMaterialParameters001}
   /SetVariables(
    Particles.Glow Color=&MultiplyLinearColor(
     Linear Color=$Particles.Color
     Scale Factor=&MakeVector4FromVectorAndFloat(
      "Vector (RGB)"=&MakeVectorFromLinearColorRGB(Color=<rgba 1 0.1356 0 1>)
      "Float (Alpha)"=&Multiply_Float(
       A=&Waveform(
        Waveform Count=::NewEnumerator1
        "Waveform [1]"=::NewEnumerator0
        "[1] Frequency"=3
        "[1] Amplitude Min/Max"=<0.3 1>
        "[2] Average"=false
        "Waveform [2]"=::NewEnumerator8
        "[2] Input Position"=0
        "[2] Phase"=$Emitter.Age
        "[2] Phase Scale"=1
        "[2] Compound Frequency"=2
        "[2] Frequency"=25
        "[2] Amplitude Min/Max"=<0.3 1>
        "[2] Scale"=1
        "[2] Bias"=0
        "[2] Exponent"=1
        "[2] Pulse Duration"=0.1
        "[2] Clip Triangle Min/Max"=<0 0>
        "[2] Random Seed"=0
       )
       B=&FloatFromCurve(FloatCurve=DataInterface() {propertyValues {Curve {
           defaultValue=0
           keys [
            {arriveTangent=9.0725, interpMode=RCIM_Cubic, leaveTangent=9.0725, tangentMode=RCTM_User, time=0, value=0}
            {arriveTangent=-0.2441, time=0.1, value=1}
            {arriveTangent=-1.1494, interpMode=RCIM_Cubic, leaveTangent=-1.1494, time=0.9, value=1}
            {arriveTangent=-7.9849, interpMode=RCIM_Cubic, leaveTangent=-7.9849, tangentMode=RCTM_User, time=0.97, value=0}
           ]
          }}})
      )
     )
    )
    Particles.Glow DP:Vector4f=&MakeVector4(X=2, Y=0, Z=0, W=0)
    Particles.Glow Size=&Multiply_Vector2D(A=$Particles.SpriteSize, B=<1.5 1.5>)
   ) {isSetParameters=true}
  ]
  rendererClasses [&NiagaraSpriteRendererProperties]
  renderers [
   {rendererClass=&NiagaraSpriteRendererProperties}
   {rendererClass=&NiagaraSpriteRendererProperties, rendererIndex=1}
   {rendererClass=&NiagaraSpriteRendererProperties, rendererIndex=2}
   {rendererClass=&NiagaraSpriteRendererProperties, rendererIndex=3}
  ]
 }
 Outro_FireParticles {
  emitterUpdateScript [
   &EmitterState(
    Life Cycle Mode=::NewEnumerator1
    Loop Behavior=::NewEnumerator1
    Loop Delay=&Add_Float(A=$System.IntroLength, B=0)
    Loop Duration=0.25
    UseLoopDelay=true
   )
   &SpawnBurst_Instantaneous(Spawn Count=10, Use Spawn Probability=false, UseLoopCountLimit=false)
   &SpawnBurst_Instantaneous(Spawn Count=10, Spawn Group=1, Use Spawn Probability=false, UseLoopCountLimit=false) {module=SpawnBurst_Instantaneous001}
   &SpawnBurst_Instantaneous(Spawn Count=10, Spawn Group=2, Use Spawn Probability=false, UseLoopCountLimit=false) {module=SpawnBurst_Instantaneous002}
   &SpawnRate(
    SpawnGroup=3
    SpawnRate=&FloatFromCurve(FloatCurve=DataInterface() {propertyValues {Curve {keys [<at 0.7 1 RCIM_Cubic>,<at 1 0>]}}}, CurveIndex=$Emitter.LoopedAge, Scale Curve=25)
    Use Spawn Probability=false
   )
   &SpawnRate(
    SpawnGroup=4
    SpawnRate=&FloatFromCurve(FloatCurve=DataInterface() {propertyValues {Curve {keys [<at 0.7 1 RCIM_Cubic>,<at 1 0>]}}}, CurveIndex=$Emitter.LoopedAge, Scale Curve=35)
    Use Spawn Probability=false
   ) {module=SpawnRate001}
  ]
  particleSpawnScript [
   &InitializeParticle(
    Color=&MergeHelper_ColorSwitch(
     0=&LerpLinearColors(
      StartColor=&RandomRangeLinearColor(Minimum=<rgba 1 0.0713 0 1>, Maximum=<rgba 1 0.4162 0 1>)
      EndColor=&RandomRangeLinearColor(Minimum=<rgba 0.1354 0.0097 0 2>, Maximum=<rgba 0.099 0.0263 0 2>)
      LerpFactor=&UniformAOrBFloat(A=0, Evaluation Type=::NewEnumerator0, Recalculate Random Each Loop=true, Override Seed=false)
     )
     1=<rgba 1 1 1 1>
     2=<rgba 3 0.7919 0.3503 1>
     3=<rgba 1 1 1 1>
     4=&RandomRangeLinearColor(Minimum=<rgba 0.4896 0.4896 0.4896 1>, Maximum=<rgba 1 1 1 1>)
     Index=&SpawnGroup()
     Num=::NewEnumerator3
    )
    Lifetime=&MergeHelper_FloatSwitch(
     0=&RandomRangeFloat(Minimum=1.5, Maximum=2.5)
     1=&RandomRangeFloat(Minimum=1.5, Maximum=2.5)
     2=&RandomRangeFloat(Minimum=1.5, Maximum=2)
     3=&RandomRangeFloat(Minimum=0.8, Maximum=1.1)
     4=&RandomRangeFloat(Minimum=0.8, Maximum=1.1)
     Index=&SpawnGroup()
     Num=::NewEnumerator3
    )
    Sprite Rotation Mode=::NewEnumerator2
    Sprite Size=&Multiply_Vector2D_ByFloat(
     Float=&FN_FrontEndFloatModifier(Gameplay=1, Front End=$User.FrontEndScale)
     Vector2D=&MergeHelper_Vector2Switch(
      0=&Multiply_Vector2D_ByFloat(Float=&RandomRangeFloat(Minimum=1, Maximum=2), Vector2D=<5 5>)
      1=&Multiply_Vector2D_ByFloat(Float=&RandomRangeFloat(Minimum=1, Maximum=2), Vector2D=<5 5>)
      2=&RandomRangeVector2D(Minimum=<0.3 2>, Maximum=<0.5 1>)
      3=&Vector2DFromFloat(Value=&Multiply_Float(A=120, B=&Multiply_Float(A=0.15, B=0.25)))
      4=&Vector2DFromFloat(Value=&Multiply_Float(A=10, B=&Multiply_Float(A=0.4, B=0.6)))
      Index=&SpawnGroup()
      Num=::NewEnumerator3
     )
    )
    Sprite Size Mode=::NewEnumerator4
    UsePositionOffset=false
    Write Material Random=true
   )
   &ShapeLocation(MaskSpawnGroup=false, Override Seed=false, Sphere Radius=10, UseOwnerScale=true)
   &AddVelocity(
    Rotation Coordinate Space=::Simulation
    Velocity=&NormalizeVector(
     Vector To Normalize=&AddVectorToPosition(
      Position=&Multiply_VectorByFloat(Vector=&ReturnCameraVector(Camera Vectors=::NewEnumerator3, Coordinate Space=::Simulation), Float=1)
      Vector=&Multiply_VectorByFloat(Vector=&ReturnCameraVector(Camera Vectors=::NewEnumerator2, Coordinate Space=::Simulation), Float=1)
     )
     Safe Normalize=true
     Safe Fallback Threshold=0
     Safe Fallback Vector=<0 0 0.001>
    )
    Velocity Speed Scale=&Multiply_Float(
     A=&RandomRangeFloat(Minimum=25, Maximum=50)
     B=&MergeHelper_FloatSwitch(
      0=&RandomRangeFloat(Minimum=0.1, Maximum=0.3)
      1=&RandomRangeFloat(Minimum=0.1, Maximum=0.3)
      2=&RandomRangeFloat(Minimum=0.1, Maximum=0.2)
      3=&RandomRangeFloat(Minimum=0.2, Maximum=0.4)
      4=&RandomRangeFloat(Minimum=0.2, Maximum=0.4)
      Index=&SpawnGroup()
      Num=::NewEnumerator3
     )
    )
   ) {module=AddVelocity001}
   &DynamicMaterialParameters(
    Index 0 Param 2=&MergeHelper_FloatSwitch(0=2, 1=0, 2=0, 3=0, 4=0, Index=&SpawnGroup(), Num=::NewEnumerator3)
    Index 0 Param 4=&MergeHelper_FloatSwitch(0=0, 1=0, 2=0, 3=2.5, 4=2.5, Index=&SpawnGroup(), Num=::NewEnumerator3)
    Module.Index0Float0Write=true
    Module.Index0Float1Write=true
    Module.Index0Float2Write=true
    Module.Index0Float3Write=true
   ) {module=DynamicMaterialParameters001}
   /SetVariables(Particles.VisibilityTag=&SpawnGroup()) {isSetParameters=true, module=SetVariables_#8}
   /SetVariables(Particles.AORandFloat:NiagaraFloat=&UniformAOrBFloat(A=0, Distribution Weight=0.3, Evaluation Type=::NewEnumerator0, Recalculate Random Each Loop=true, Override Seed=false)) {isSetParameters=true, module=SetVariables_#9}
   &CameraOffset(Camera Offset Amount=&MergeHelper_FloatSwitch(0=0, 1=0, 2=5, 3=15, 4=0, Index=&SpawnGroup(), Num=::NewEnumerator3))
   /SetVariables() {isSetParameters=true}
  ]
  particleUpdateScript [
   &ParticleState()
   &SpriteRotationRate(Rotation Rate=&MergeHelper_FloatSwitch(
     0=0
     1=0
     2=0
     3=&RandomRangeFloat(Minimum=-300, Maximum=300)
     4=&RandomRangeFloat(Minimum=-300, Maximum=300)
     Index=&SpawnGroup()
     Num=::NewEnumerator3
    ))
   &SubUVAnimation(
    End Frame Range Override=12
    Evaluation Type=::NewEnumerator0
    "Sprite Renderer "=DataInterface() {propertyValues {SpriteRenderer=&self:Outro_FireParticles_0.NiagaraSpriteRendererProperties_0}}
    Start Frame Range Override=8
    SubUV Animation Mode:ENiagara_SubUVLookupModeV2=::NewEnumerator2
    UseEndFrame=false
    UseStartFrame=false
   )
   &SubUVAnimation(
    End Frame Range Override=11
    Evaluation Type=::NewEnumerator0
    "Sprite Renderer "=DataInterface() {propertyValues {SpriteRenderer=&self:Outro_FireParticles_0.NiagaraSpriteRendererProperties_1}}
    Start Frame Range Override=1
    SubUV Animation Mode:ENiagara_SubUVLookupModeV2=::NewEnumerator2
    UseEndFrame=true
    UseStartFrame=true
   ) {module=SubUVAnimation001}
   &SubUVAnimation(
    End Frame Range Override=12
    Evaluation Type=::NewEnumerator0
    "Sprite Renderer "=DataInterface() {propertyValues {SpriteRenderer=&self:Outro_FireParticles_0.NiagaraSpriteRendererProperties_14}}
    Start Frame Range Override=8
    SubUV Animation Mode:ENiagara_SubUVLookupModeV2=::NewEnumerator2
    UseEndFrame=false
    UseStartFrame=false
   ) {module=SubUVAnimation003}
   &SubUVAnimation(
    End Frame Range Override=12
    Evaluation Type=::NewEnumerator0
    "Sprite Renderer "=DataInterface() {propertyValues {SpriteRenderer=&self:Outro_FireParticles_0.NiagaraSpriteRendererProperties_12}}
    Start Frame Range Override=8
    SubUV Animation Mode:ENiagara_SubUVLookupModeV2=::NewEnumerator2
    UseEndFrame=true
    UseStartFrame=true
   ) {module=SubUVAnimation002}
   &OffsetPosition(Position Offset=$System.BurstPosition)
   &ScaleSpriteSize(
    Scale Sprite Size Mode=::NewEnumerator0
    Uniform Scale Factor=&MergeHelper_FloatSwitch(
     0=&FloatFromCurve(
      FloatCurve=DataInterface() {propertyValues {Curve {
         defaultValue=1
         keys [{arriveTangent=-1.9355, interpMode=RCIM_Cubic, time=0, value=1}, {interpMode=RCIM_Cubic, tangentMode=RCTM_User, time=1, value=0}]
        }}}
      Scale Curve=0.6
     )
     1=&FloatFromCurve(FloatCurve=DataInterface() {propertyValues {Curve {keys [<at 0 0>,<at 0.1 0.5>,<at 1 0>]}}})
     2=&FloatFromCurve(
      FloatCurve=DataInterface() {propertyValues {Curve {keys [
          <at 0 0 RCIM_Cubic>
          {interpMode=RCIM_Cubic, leaveTangent=-3.0296, tangentMode=RCTM_Break, time=0.2, value=1.2}
          {arriveTangent=-2.8894, interpMode=RCIM_Cubic, leaveTangent=-2.8894, tangentMode=RCTM_User, time=1, value=0.1}
         ]}}}
      Scale Curve=1.6
     )
     3=&FloatFromCurve(
      FloatCurve=DataInterface() {propertyValues {Curve {keys [
          {arriveTangent=1.3427, interpMode=RCIM_Cubic, leaveTangent=1.3427, tangentMode=RCTM_User, time=0, value=0}
          {interpMode=RCIM_Cubic, tangentMode=RCTM_User, time=0.2, value=0.3}
          {tangentMode=RCTM_User, time=1, value=0.1}
         ]}}}
      Scale Curve=12
     )
     4=&FloatFromCurve(
      FloatCurve=DataInterface() {propertyValues {Curve {keys [
          {arriveTangent=1.8745, interpMode=RCIM_Cubic, leaveTangent=1.8745, tangentMode=RCTM_User, time=0, value=0.4734}
          {tangentMode=RCTM_User, time=1, value=1}
         ]}}}
      Scale Curve=6
     )
     Index=&SpawnGroup()
     Num=::NewEnumerator3
    )
   ) {module=ScaleSpriteSize001}
   &ScaleColor(
    Scale Alpha=&MergeHelper_FloatSwitch(
     0=&FloatFromCurve(
      FloatCurve=DataInterface() {propertyValues {Curve {
         defaultValue=0
         keys [
          {arriveTangent=-2.5249, interpMode=RCIM_Cubic, leaveTangent=-2.5249, tangentMode=RCTM_User, time=0, value=0.9847}
          {arriveTangent=-0.0405, interpMode=RCIM_Cubic, leaveTangent=-0.0405, tangentMode=RCTM_User, time=1, value=0}
         ]
        }}}
      Scale Curve=0.4
     )
     1=2
     2=1
     3=&Multiply_Float(
      A=1
      B=&FloatFromCurve(
       FloatCurve=DataInterface() {propertyValues {Curve {keys [
           {arriveTangent=3.6955, interpMode=RCIM_Cubic, leaveTangent=3.6955, tangentMode=RCTM_User, time=0, value=0}
           {interpMode=RCIM_Cubic, tangentMode=RCTM_User, time=0.5, value=1}
           {tangentMode=RCTM_User, time=1, value=0.1}
          ]}}}
       Scale Curve=&Multiply_Float(
        A=&FN_FrontEndFloatModifier(Gameplay=0.6, Front End=0.4)
        B=&Lerp_Float(A=0.2, B=0.7, Alpha=$Particles.AORandFloat)
       )
      )
     )
     4=&Multiply_Float(
      A=2
      B=&FloatFromCurve(
       FloatCurve=DataInterface() {propertyValues {Curve {keys [<at 0 1 RCIM_Cubic>, {interpMode=RCIM_Cubic, tangentMode=RCTM_User, time=0.1942, value=1}, {tangentMode=RCTM_User, time=1, value=0}]}}}
       Scale Curve=0.7
      )
     )
     Index=&SpawnGroup()
     Num=::NewEnumerator3
    )
    Scale RGB=&MergeHelper_VectorSwitch(
     0=&VectorFromFloat(Value=&Lerp_Float(
       A=0
       B=&RandomRangeFloat(Minimum=3, Maximum=10)
       Alpha=&MakeFloatFromVector(
        Vector=&VectorFromFloat(Value=&FloatFromCurve(FloatCurve=DataInterface() {propertyValues {Curve {
             defaultValue=1
             keys [
              {arriveTangent=2.9173, interpMode=RCIM_Cubic, leaveTangent=2.9173, tangentMode=RCTM_User, time=0, value=0.7799}
              {arriveTangent=-2.5249, interpMode=RCIM_Cubic, leaveTangent=-2.5249, tangentMode=RCTM_User, time=0.3924, value=0.933}
              {arriveTangent=-0.0405, interpMode=RCIM_Cubic, leaveTangent=-0.0405, tangentMode=RCTM_User, time=1, value=0}
             ]
            }}}))
        Channel=::NewEnumerator0
       )
      ))
     1=&MakeVectorFromLinearColorRGB(Color=&ColorFromCurve(ColorCurve=DataInterface() {propertyValues {
         AlphaCurve {keys []}
         BlueCurve {keys [<at 0 0 RCIM_Cubic>,<at 1 0 RCIM_Cubic>]}
         GreenCurve {keys [<at 0 0.2885 RCIM_Cubic>,<at 1 0.0015 RCIM_Cubic>]}
         RedCurve {keys [<at 0 2 RCIM_Cubic>,<at 1 0.0262 RCIM_Cubic>]}
        }}))
     2=<1 1 1>
     3=&Multiply_VectorByFloat(
      Vector=&MakeVectorFromLinearColorRGB(Color=&LerpLinearColors(
        StartColor=&ColorFromCurve(ColorCurve=DataInterface() {propertyValues {
           AlphaCurve {keys []}
           BlueCurve {keys [<at 0.6023 0.0007 RCIM_Cubic>]}
           GreenCurve {keys [<at 0.6023 0.3353 RCIM_Cubic>]}
           RedCurve {keys [<at 0.6023 4 RCIM_Cubic>]}
          }})
        EndColor=<rgba 0.0938 0.0138 0.003 1>
        LerpFactor=$Particles.AORandFloat
       ))
      Float=1
     )
     4=&Multiply_VectorByFloat(
      Vector=&MakeVectorFromLinearColorRGB(Color=&ColorFromCurve(ColorCurve=DataInterface() {propertyValues {
          AlphaCurve {keys []}
          BlueCurve {keys [<at 0.0268 0.0698>,<at 0.1891 0.0006>,<at 0.4587 0>,<at 0.7901 0.0346 RCIM_Cubic>]}
          GreenCurve {keys [<at 0.0268 0.3962>,<at 0.1891 0.0264>,<at 0.4587 0.0243>,<at 0.7901 0.2452 RCIM_Cubic>]}
          RedCurve {keys [<at 0.0268 2>,<at 0.1891 0.1146>,<at 0.4587 0.1146>,<at 0.7901 2 RCIM_Cubic>]}
         }}))
      Float=&FloatFromCurve(FloatCurve=DataInterface() {propertyValues {Curve {keys [<at 0.1 2>,<at 0.519 1>,<at 1 0>]}}})
     )
     Index=&SpawnGroup()
     Num=::NewEnumerator3
    )
    ScaleA=true
    ScaleRGB=true
   )
   &CurlNoiseForce(
    Bias Noise=false
    Noise Frequency=25
    Noise Strength=&FloatFromCurve(
     FloatCurve=DataInterface() {propertyValues {Curve {keys [{arriveTangent=2.1957, interpMode=RCIM_Cubic, leaveTangent=2.1957, tangentMode=RCTM_User, time=0, value=0}, <at 1 1>]}}}
     Scale Curve=&MergeHelper_FloatSwitch(0=1000, 1=1000, 2=1500, 3=750, 4=750, Index=&SpawnGroup(), Num=::NewEnumerator3)
    )
    Pan Noise=true
    Pan Noise Field=<0 0 1>
    Random Seed:NiagaraInt32=&MergeHelper_IntSwitch(0=11, 1=12, 2=13, 3=14, 4=15, 5=?, Index=&SpawnGroup(), Num=::NewEnumerator3)
   )
   &ScaleVelocity(Velocity Scale=&VectorFromFloat(Value=&Multiply_Float(
      A=&FloatFromCurve(
       FloatCurve=DataInterface() {propertyValues {Curve {keys [
           {arriveTangent=-0.1413, interpMode=RCIM_Cubic, time=0, value=0.8}
           {arriveTangent=-0.1274, interpMode=RCIM_Cubic, leaveTangent=-0.1274, tangentMode=RCTM_User, time=0.4, value=0.2}
           <at 0.9984 0 RCIM_Cubic>
          ]}}}
       Scale Curve=&MergeHelper_FloatSwitch(0=3.6, 1=3.6, 2=6, 3=6, 4=6, Index=&SpawnGroup(), Num=::NewEnumerator3)
      )
      B=&FN_FrontEndFloatModifier(Gameplay=1, Front End=$User.FrontEndScale)
     )))
   &Drag(
    Drag=&MergeHelper_FloatSwitch(0=5, 1=5, 2=4, 3=6, 4=3, Index=&SpawnGroup(), Num=::NewEnumerator3)
    Use Linear Drag=true
    Use Rotational Drag=true
   )
   &SolveForcesAndVelocity(Clamp Velocity=false, Limit Acceleration=false)
   &ScaleSpriteSizeBySpeed(Max Scale Factor=&MergeHelper_Vector2Switch(0=<1 1>, 1=<1 1>, 2=<1 3>, 3=<1 1>, 4=<1 1>, Index=&SpawnGroup(), Num=::NewEnumerator3))
   &FN_Sprite_Attributes(
    Color=&Multiply_LinearColorByFloat(Linear Color=$Particles.Color, Float=0.304)
    Create Color Attribute=true
    Create Dynamic Material Parameter Attributes=true
    Create SpriteSize Attribute=true
    Index 0 Param 1=3
    Index 0 Param 2=1
    Index 0 Param 3=1
    Index 0 Param 4=1
    Module Output:ENiagara_AttributeSamplingApplyOutput=::NewEnumerator1
    Module.Index0Float0Write=true
    Module.Index0Float1Write=true
    Module.Index0Float2Write=true
    Module.Index0Float3Write=true
    SpriteSize=&Multiply_Vector2D_ByFloat(Float=1.4, Vector2D=$Particles.SpriteSize)
   ) {module=FN_Sprite_Attributes001}
   &FN_Sprite_Attributes(
    Color=&MakeLinearColorFromVectorAndFloat(
     "Vector (RGB)"=&MakeVectorFromLinearColorRGB(Color=<rgba 2 0.2885 0 1>)
     "Float (Alpha)"=&Multiply_Float(
      A=&FN_FrontEndFloatModifier(Gameplay=1, Front End=0.2)
      B=&FloatFromCurve(
       FloatCurve=DataInterface() {propertyValues {Curve {keys [
           {interpMode=RCIM_Cubic, tangentMode=RCTM_User, time=0, value=1}
           {arriveTangent=-2.7437, interpMode=RCIM_Cubic, leaveTangent=-2.7437, tangentMode=RCTM_User, time=1, value=0}
          ]}}}
       Scale Curve=0.25
      )
     )
    )
    Create Color Attribute=true
    Create SpriteSize Attribute=true
    Module Output:ENiagara_AttributeSamplingApplyOutput=::NewEnumerator1
    SpriteSize=&Multiply_Vector2D_ByFloat(
     Float=&MakeFloatFromVector2d(Vector 2D=&Vector2DFromFloat(Value=&FloatFromCurve(FloatCurve=DataInterface() {propertyValues {Curve {keys [<at 0 0>,<at 0.1 0.5>,<at 1 0>]}}})))
     Vector2D=&Multiply_Vector2D_ByFloat(Float=5, Vector2D=$Particles.SpriteSize)
    )
   ) {module=FN_Sprite_Attributes002}
   &FN_Sprite_Attributes(
    Color=&MakeLinearColorFromVectorAndFloat(
     "Vector (RGB)"=&MakeVectorFromLinearColorRGB(Color=<rgba 1 0.0833 0 0.2>)
     "Float (Alpha)"=&FloatFromCurve(
      FloatCurve=DataInterface() {propertyValues {Curve {keys [
          {interpMode=RCIM_Cubic, tangentMode=RCTM_User, time=0, value=1}
          {arriveTangent=-2.7437, interpMode=RCIM_Cubic, leaveTangent=-2.7437, tangentMode=RCTM_User, time=1, value=0}
         ]}}}
      Scale Curve=&Multiply_Float(A=0.2, B=1)
     )
    )
    Create Color Attribute=true
    Create Dynamic Material Parameter Attributes=true
    Create SpriteSize Attribute=true
    Index 0 Param 1=2
    Index 0 Param 4=2.5
    Module Output:ENiagara_AttributeSamplingApplyOutput=::NewEnumerator1
    Module.Index0Float0Write=true
    Module.Index0Float1Write=true
    Module.Index0Float2Write=true
    Module.Index0Float3Write=true
    SpriteSize=&Multiply_Vector2D_ByFloat(Float=5, Vector2D=&Multiply_Vector2D(A=$Particles.SpriteSize, B=<1 0.8>))
   )
   &FN_Sprite_Attributes(
    Color=&MakeLinearColorFromVectorAndFloat(
     "Vector (RGB)"=&MakeVectorFromLinearColorRGB(Color=&LerpLinearColors(StartColor=<rgba 1 0.2 0 0.5>, EndColor=<rgba 0.1719 0.0344 0 1>, LerpFactor=$Particles.AORandFloat))
     "Float (Alpha)"=&Multiply_Float(A=0.14, B=&FN_FrontEndFloatModifier(Gameplay=0.5, Front End=0.1))
    )
    Create Color Attribute=true
    Create SpriteSize Attribute=true
    Module Output:ENiagara_AttributeSamplingApplyOutput=::NewEnumerator1
    SpriteSize=&Multiply_Vector2D(A=$Particles.SpriteSize, B=<2 2>)
   ) {module=FN_Sprite_Attributes003}
   &FN_Sprite_Attributes(
    Color=&MakeLinearColorFromVectorAndFloat(
     "Vector (RGB)"=&MakeVectorFromLinearColorRGB(Color=<rgba 10 0.6212 0.1562 0.5>)
     "Float (Alpha)"=&FloatFromCurve(
      FloatCurve=DataInterface() {propertyValues {Curve {keys [<at 0 0.5428 RCIM_Cubic>, {arriveTangent=-0.5428, interpMode=RCIM_Cubic, leaveTangent=-0.5428, time=0.1667, value=1}, <at 1 0 RCIM_Cubic>]}}}
      Scale Curve=0.7
     )
    )
    Create Color Attribute=true
    Create Dynamic Material Parameter Attributes=true
    Create SpriteSize Attribute=true
    Index 0 Param 1=3
    Index 0 Param 2=1
    Index 0 Param 3=1
    Index 0 Param 4=1
    Module Output:ENiagara_AttributeSamplingApplyOutput=::NewEnumerator1
    Module.Index0Float0Write=true
    Module.Index0Float1Write=true
    Module.Index0Float2Write=true
    Module.Index0Float3Write=true
    SpriteSize=&Multiply_Vector2D_ByFloat(Float=5, Vector2D=$Particles.SpriteSize)
   ) {module=FN_Sprite_Attributes004}
   &FN_Sprite_Attributes(
    Color=&MakeLinearColorFromVectorAndFloat(
     "Vector (RGB)"=&MakeVectorFromLinearColorRGB(Color=<rgba 1 0.2 0 0.5>)
     "Float (Alpha)"=&Multiply_Float(
      A=&FloatFromCurve(
       FloatCurve=DataInterface() {propertyValues {Curve {keys [<at 0 0.5428 RCIM_Cubic>, {arriveTangent=-0.5428, interpMode=RCIM_Cubic, leaveTangent=-0.5428, time=0.1667, value=1}, <at 1 0 RCIM_Cubic>]}}}
       Scale Curve=0.4
      )
      B=&FN_FrontEndFloatModifier(Gameplay=0.3, Front End=0.1)
     )
    )
    Create Color Attribute=true
    Create SpriteSize Attribute=true
    Module Output:ENiagara_AttributeSamplingApplyOutput=::NewEnumerator1
    SpriteSize=&Multiply_Vector2D_ByFloat(Float=3, Vector2D=$Particles.SpriteSize)
   ) {module=FN_Sprite_Attributes005}
  ]
  rendererClasses [&NiagaraSpriteRendererProperties]
  renderers [
   {rendererClass=&NiagaraSpriteRendererProperties}
   {rendererClass=&NiagaraSpriteRendererProperties, rendererIndex=1}
   {rendererClass=&NiagaraSpriteRendererProperties, rendererIndex=2}
   {rendererClass=&NiagaraSpriteRendererProperties, rendererIndex=3}
   {rendererClass=&NiagaraSpriteRendererProperties, rendererIndex=4}
   {rendererClass=&NiagaraSpriteRendererProperties, rendererIndex=5}
   {rendererClass=&NiagaraSpriteRendererProperties, rendererIndex=6}
   {rendererClass=&NiagaraSpriteRendererProperties, rendererIndex=7}
   {rendererClass=&NiagaraSpriteRendererProperties, rendererIndex=8}
   {rendererClass=&NiagaraSpriteRendererProperties, rendererIndex=9}
   {rendererClass=&NiagaraSpriteRendererProperties, rendererIndex=10}
  ]
 }
 Outro_Flames {
  emitterUpdateScript [
   &EmitterState(
    Life Cycle Mode=::NewEnumerator1
    Loop Behavior=::NewEnumerator1
    Loop Delay=&Add_Float(A=$System.IntroLength, B=-0.4)
    Loop Duration=0.25
    UseLoopDelay=true
   )
   &SpawnBurst_Instantaneous(Use Spawn Probability=false, UseLoopCountLimit=false)
   &SpawnBurst_Instantaneous(Spawn Count=2, Spawn Group=1, Use Spawn Probability=false, UseLoopCountLimit=false) {module=SpawnBurst_Instantaneous001}
   &SpawnBurst_Instantaneous(Spawn Count=3, Spawn Group=2, Use Spawn Probability=false, UseLoopCountLimit=false) {module=SpawnBurst_Instantaneous002}
  ]
  particleSpawnScript [
   /SetVariables(Particles.VisibilityTag=&SpawnGroup()) {isSetParameters=true, module=SetVariables_#10}
   &InitializeParticle(
    Lifetime=&Multiply_Float(A=2, B=0.9)
    Position Offset=$System.BurstPosition
    Sprite Rotation Angle=40
    Sprite Rotation Mode=::NewEnumerator0
    Sprite Size=&Multiply_Vector2D_ByFloat(
     Float=&FN_FrontEndFloatModifier(Gameplay=1, Front End=$User.FrontEndScale)
     Vector2D=&Multiply_Vector2D(
      A=&MergeHelper_Vector2Switch(0=<15 25>, 1=<20 20>, 2=<25 25>, 3=?, 4=?, Index=&SpawnGroup(), Num=::NewEnumerator1)
      B=<1.5 1.5>
     )
    )
    Sprite Size Mode=::NewEnumerator4
    UsePositionOffset=true
    Write Material Random=true
   )
  ]
  particleUpdateScript [
   /SetVariables(Particles.PivotOffset:Vector2f=&Add_Vector2(
     A:Vector2f=<0.5 0.5>
     B:Vector2f=&Lerp_Vector2(
      A=<-0.0587 0>
      B=<-0.0587 0.3>
      Alpha=&FloatFromCurve(FloatCurve=DataInterface() {propertyValues {Curve {
          defaultValue=0
          keys [
           {arriveTangent=2.2881, interpMode=RCIM_Cubic, leaveTangent=2.2881, tangentMode=RCTM_User, time=0.25, value=0}
           {interpMode=RCIM_Cubic, tangentMode=RCTM_User, time=1, value=1}
          ]
         }}})
     )
    )) {isSetParameters=true, module=SetVariables_#11}
   &ParticleState()
   &Drag(Drag=5, Use Linear Drag=true, Use Rotational Drag=true)
   &SolveForcesAndVelocity(Clamp Velocity=false, Limit Acceleration=false)
   &ScaleColor(
    Scale Alpha=&FloatFromCurve(FloatCurve=DataInterface() {propertyValues {Curve {
        defaultValue=0
        keys [
         {arriveTangent=9.0725, interpMode=RCIM_Cubic, leaveTangent=9.0725, tangentMode=RCTM_User, time=0.31, value=0}
         {arriveTangent=-0.2441, time=0.45, value=1}
         {arriveTangent=-1.9231, interpMode=RCIM_Cubic, leaveTangent=-1.9231, time=0.75, value=1}
         {arriveTangent=-7.9849, interpMode=RCIM_Cubic, leaveTangent=-7.9849, tangentMode=RCTM_User, time=0.97, value=0}
        ]
       }}})
    Scale RGB=&Multiply_VectorByFloat(
     Vector=&MakeVectorFromLinearColorRGB(Color=&MergeHelper_ColorSwitch(
       0=<rgba 1 0.7517 0.007 1>
       1=<rgba 1 0.5457 0.0006 1>
       2=<rgba 1 0.1356 0 1>
       3=?
       Index=&SpawnGroup()
       Num=::NewEnumerator1
      ))
     Float=1
    )
    ScaleA=false
    ScaleRGB=true
   )
   &CameraOffset(Camera Offset Amount=10)
   &SubUVAnimation(
    Play Rate=0.5
    Random Start Frame=true
    "Sprite Renderer "=DataInterface() {propertyValues {SpriteRenderer=&self:Outro_Flames_0.NiagaraSpriteRendererProperties_1}}
    Start Frame Offset:NiagaraInt32=&RandomRangeInt(Minimum=0, Maximum=63)
    SubUV Animation Mode:ENiagara_SubUVLookupModeV2=::NewEnumerator3
    UseEndFrame=true
    UsePlayRateScale=false
    UseStartFrame=false
   )
   &SubUVAnimation(
    Play Rate=0.5
    Playback Scale Curve=DataInterface() {propertyValues {Curve {keys [
        {interpMode=RCIM_Cubic, tangentMode=RCTM_User, time=0.25, value=1}
        {arriveTangent=-1.2656, interpMode=RCIM_Cubic, leaveTangent=-1.2656, tangentMode=RCTM_User, time=1, value=0}
       ]}}}
    Random Start Frame=true
    "Sprite Renderer "=DataInterface() {propertyValues {SpriteRenderer=&self:Outro_Flames_0.NiagaraSpriteRendererProperties_0}}
    Start Frame Offset:NiagaraInt32=&RandomRangeInt(Minimum=0, Maximum=63)
    SubUV Animation Mode:ENiagara_SubUVLookupModeV2=::NewEnumerator3
    UseEndFrame=true
    UsePlayRateScale=false
    UseStartFrame=false
   ) {module=SubUVAnimation001}
   &SubUVAnimation(
    Play Rate=0.5
    Playback Scale Curve=DataInterface() {propertyValues {Curve {keys [
        {interpMode=RCIM_Cubic, tangentMode=RCTM_User, time=0.25, value=1}
        {arriveTangent=-1.2656, interpMode=RCIM_Cubic, leaveTangent=-1.2656, tangentMode=RCTM_User, time=1, value=0}
       ]}}}
    Random Start Frame=true
    "Sprite Renderer "=DataInterface() {propertyValues {SpriteRenderer=&self:Outro_Flames_0.NiagaraSpriteRendererProperties_2}}
    Start Frame Offset:NiagaraInt32=&RandomRangeInt(Minimum=0, Maximum=63)
    SubUV Animation Mode:ENiagara_SubUVLookupModeV2=::NewEnumerator3
    UseEndFrame=true
    UsePlayRateScale=false
    UseStartFrame=false
   ) {module=SubUVAnimation002}
   &DynamicMaterialParameters(
    Index 0 Param 1=&FloatFromCurve(FloatCurve=DataInterface() {propertyValues {Curve {
        defaultValue=3
        keys [
         {interpMode=RCIM_Cubic, tangentMode=RCTM_User, time=0, value=3}
         {arriveTangent=-7.9406, interpMode=RCIM_Cubic, tangentMode=RCTM_Break, time=0.35, value=0}
         {tangentMode=RCTM_User, time=1, value=3}
        ]
       }}})
    Index 0 Param 2=&RandomRangeFloat(Minimum=0, Maximum=1)
    Index 0 Param 3=1
    Index 0 Param 4=1
    Module.Index0Float0Write=true
    Module.Index0Float1Write=true
    Module.Index0Float2Write=true
    Module.Index0Float3Write=true
   ) {module=DynamicMaterialParameters001}
   /SetVariables(
    Particles.Glow Color=&MultiplyLinearColor(
     Linear Color=$Particles.Color
     Scale Factor=&MakeVector4FromVectorAndFloat(
      "Vector (RGB)"=&MakeVectorFromLinearColorRGB(Color=<rgba 1 0.1356 0 1>)
      "Float (Alpha)"=&Multiply_Float(
       A=&Waveform(
        Waveform Count=::NewEnumerator1
        "Waveform [1]"=::NewEnumerator0
        "[1] Frequency"=3
        "[1] Amplitude Min/Max"=<0.3 1>
        "[2] Average"=false
        "Waveform [2]"=::NewEnumerator8
        "[2] Input Position"=0
        "[2] Phase"=$Emitter.Age
        "[2] Phase Scale"=1
        "[2] Compound Frequency"=2
        "[2] Frequency"=25
        "[2] Amplitude Min/Max"=<0.3 1>
        "[2] Scale"=1
        "[2] Bias"=0
        "[2] Exponent"=1
        "[2] Pulse Duration"=0.1
        "[2] Clip Triangle Min/Max"=<0 0>
        "[2] Random Seed"=0
       )
       B=&FloatFromCurve(
        FloatCurve=DataInterface() {propertyValues {Curve {
           defaultValue=3
           keys [
            {arriveTangent=0.3112, interpMode=RCIM_Cubic, leaveTangent=0.3112, tangentMode=RCTM_User, time=0, value=0}
            {arriveTangent=0.2522, interpMode=RCIM_Cubic, leaveTangent=-2.978, tangentMode=RCTM_Break, time=0.35, value=1}
            {tangentMode=RCTM_User, time=1, value=0}
           ]
          }}}
        Scale Curve=&FN_FrontEndFloatModifier(Gameplay=0.3, Front End=0.1)
       )
      )
     )
    )
    Particles.Glow DP:Vector4f=&MakeVector4(X=2, Y=0, Z=0, W=0)
    Particles.Glow Pivot Offset:Vector2f=&Add_Vector2(
     A:Vector2f=<0.5 0.6>
     B:Vector2f=&Lerp_Vector2(
      A=<-0.0587 0.1379>
      B=<-0.0587 0.333>
      Alpha=&FloatFromCurve(FloatCurve=DataInterface() {propertyValues {Curve {
          defaultValue=0
          keys [
           {arriveTangent=2.2881, interpMode=RCIM_Cubic, leaveTangent=2.2881, tangentMode=RCTM_User, time=0.25, value=0}
           {interpMode=RCIM_Cubic, tangentMode=RCTM_User, time=1, value=1}
          ]
         }}})
     )
    )
    Particles.Glow Size=&Multiply_Vector2D(A=$Particles.SpriteSize, B=<1.5 1.1>)
   ) {isSetParameters=true}
  ]
  rendererClasses [&NiagaraSpriteRendererProperties]
  renderers [
   {rendererClass=&NiagaraSpriteRendererProperties}
   {rendererClass=&NiagaraSpriteRendererProperties, rendererIndex=1}
   {rendererClass=&NiagaraSpriteRendererProperties, rendererIndex=2}
   {rendererClass=&NiagaraSpriteRendererProperties, rendererIndex=3}
  ]
 }
}
summary {
 systemName=NS_ReactFX_TalonScone
 emitters [
  {emitterName=Heart, rendererClasses [&NiagaraSpriteRendererProperties]}
  {emitterName=Intro_FireTrail, rendererClasses [&NiagaraRibbonRendererProperties]}
  {emitterName=HeartFire, rendererClasses [&NiagaraSpriteRendererProperties]}
  {emitterName=Outro_FireParticles, rendererClasses [&NiagaraSpriteRendererProperties]}
  {emitterName=Outro_Flames, rendererClasses [&NiagaraSpriteRendererProperties]}
  {emitterName=Intro_Flame, rendererClasses [&NiagaraSpriteRendererProperties]}
  {emitterName=Intro_FireParticles, rendererClasses [&NiagaraSpriteRendererProperties]}
  {emitterName=FireParticles, rendererClasses [&NiagaraSpriteRendererProperties]}
 ]
 userVariables [
  User.bIsFrontEnd {type=NiagaraBool}
  User.bIsFrontEndPreview {type=NiagaraBool}
  User.FrontEndScale {type=NiagaraFloat, defaultValue=0.6, description=""}
 ]
}
