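# Niagara cosmetic companion reaction effect themed on boba, judging by its TapiTime name and particle attributes, with a petting-reaction sound.
# Spawn sprays droplets upward while Intro_Rebuild animates a squash-and-stretch main sprite with bobbing, spinning boba that travel toward a burst position.
# Pop fires three sprite bursts there when the intro length ends.
# Age curves drive motion, color and material parameters, with a scratch gimbal-lock module using the camera vector.
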NiagaraSystem /CosmeticCompanions/FX/React/NS_ReactFX_TapiTime
dependencies {
 usedDataInterfaces [
  &NiagaraDataInterfaceSpriteRendererInfo
  &NiagaraDataInterfaceCurve
  &NiagaraDataInterfaceVectorCurve
  &NiagaraDataInterfaceAudioPlayer
  &NiagaraDataInterfaceVector2DCurve
  &NiagaraDataInterfaceCamera
  &NiagaraDataInterfacePlatformSet
 ]
 usedDynamicInputs [
  &FN_FrontEndVector3Modifier
  /Game/Effects/Niagara/DynamicInputs/Logic/FN_PickLinearColorDependOnPlatform
  &self:ScratchDynamicInput_01
 ]
 usedModules [
  &self:SystemSpawnScript.SystemScriptSource.SystemScriptGraph.NiagaraNodeAssignment_0.SetVariables_#1
  &self:SystemSpawnScript.SystemScriptSource.SystemScriptGraph.NiagaraNodeAssignment_5.SetVariables_#2
  &self:SystemSpawnScript.SystemScriptSource.SystemScriptGraph.NiagaraNodeAssignment_1.SetVariables_#3
  &self:SystemSpawnScript.SystemScriptSource.SystemScriptGraph.NiagaraNodeAssignment_2.SetVariables_#4
  &SystemState
  &self:SystemSpawnScript.SystemScriptSource.SystemScriptGraph.NiagaraNodeAssignment_3.SetVariables_#5
  &self:SystemSpawnScript.SystemScriptSource.SystemScriptGraph.NiagaraNodeAssignment_7.SetVariables_#6
  &self:SystemSpawnScript.SystemScriptSource.SystemScriptGraph.NiagaraNodeAssignment_6.SetVariables_#7
  &self:SystemSpawnScript.SystemScriptSource.SystemScriptGraph.NiagaraNodeAssignment_4.SetVariables_#8
 ]
}
emitters {
 Intro_Rebuild {
  emitterSpawnScript [
   &InitializeAudioPlayer(Audio Player=DataInterface() {propertyValues {
      Attenuation=null
      Concurrency=null
      MaxPlaysPerTick=10
      SoundToPlay=/CosmeticCompanions/Assets/Audio_Shared/Interaction/Petting/SC_Mimosa_Interaction_React_Default
      bAllowLoopingOneShotSounds=false
      bLimitPlaysPerTick=true
      bOnlyActiveDuringGameplay=false
      bStopWhenComponentIsDestroyed=true
      ConfigurationUserParameter {parameter {name=None, typeDefHandle {registeredTypeIndex=1}, varData []}}
      ParameterNames []
     }})
   /Niagara/Modules/Audio/PlayAudio() {module=PlayAudio001}
  ]
  emitterUpdateScript [
   &EmitterState()
   &SpawnBurst_Instantaneous(Use Spawn Probability=false, UseLoopCountLimit=false)
   &SpawnBurst_Instantaneous(Spawn Count=11, Spawn Group=1, Use Spawn Probability=false, UseLoopCountLimit=false) {module=SpawnBurst_Instantaneous001}
   /SetVariables(
    Emitter.SpinIntensity:NiagaraFloat=&FloatFromCurve(
     FloatCurve=DataInterface() {propertyValues {Curve {keys [
         {interpMode=RCIM_Cubic, tangentMode=RCTM_User, time=0.3, value=0.1}
         {
          arriveTangent=2.8517
          arriveTangentWeight=0.6537
          interpMode=RCIM_Cubic
          leaveTangent=1.175
          leaveTangentWeight=0.5771
          tangentMode=RCTM_Break
          tangentWeightMode=RCTWM_WeightedBoth
          time=0.6
          value=5
         }
        ]}}}
     CurveIndex=&NormalizeFloat(Value=$Emitter.Age, Min=0, Max=1)
    )
    Emitter.VelocityForce:NiagaraFloat=&FloatFromCurve(
     FloatCurve=DataInterface() {propertyValues {Curve {keys [{arriveTangentWeight=0.2333, interpMode=RCIM_Cubic, leaveTangentWeight=0.1, tangentMode=RCTM_User, time=0.75, value=0}, <at 0.78 1>]}}}
     CurveIndex=$Emitter.Age
    )
   ) {isSetParameters=true}
  ]
  particleSpawnScript [
   /SetVariables(Particles.VisibilityTag=&SpawnGroup()) {isSetParameters=true}
   &InitializeParticle(
    Color Mode=::NewEnumerator0
    Lifetime=&MaskFloat(
     Mask Source=0
     Mask Parameter=?
     Mask=0
     Masked Float=$System.IntroParticleLife
     Passthrough Float=&MaskFloat(
      Mask Source=0
      Mask Parameter=?
      Mask=1
      Masked Float=&Add_Float(A=$System.IntroParticleLife, B=&RandomRangeFloat(Minimum=0.2, Maximum=0.3))
      Passthrough Float=&RandomRangeFloat(Minimum=0.15, Maximum=0.3)
     )
    )
    Sprite Size Mode=::NewEnumerator3
    Uniform Sprite Size=20
    UsePositionOffset=false
    Write Material Random=true
   )
   &ShapeLocation(
    MaskSpawnGroup=false
    Non Uniform Scale=<1 1 0.7>
    Override Seed=false
    Sphere Radius=&MaskFloat(Mask Source=0, Mask Parameter=?, Mask=1, Masked Float=8, Passthrough Float=0)
    UseOwnerScale=true
   )
   /SetVariables(Particles.BobaColor:LinearColor=<rgba 1 1 1 50>, Particles.BobaSize:Vector2f=<3.5 3.5>) {isSetParameters=true}
   /SetVariables(
    Particles.DropletColor:LinearColor=$Particles.Color
    Particles.DropletSubUV:NiagaraFloat=&RandomRangeFloat(Minimum=0, Maximum=3)
   ) {isSetParameters=true}
  ]
  particleUpdateScript [
   &ParticleState()
   /SetVariables(Particles.InitDroplet Dynp 00:Vector4f=&MakeVector4(
     X=0
     Y=&FloatFromCurve(FloatCurve=DataInterface() {propertyValues {Curve {defaultValue=1, keys [{interpMode=RCIM_Cubic, tangentMode=RCTM_User, time=0.0693, value=0}, {tangentMode=RCTM_User, time=0.1074, value=2}]}}})
     Z=0
     W=0
    )) {isSetParameters=true}
   &DynamicMaterialParameters(
    Index 0 Param 1=&FloatFromCurve(FloatCurve=DataInterface() {propertyValues {Curve {keys [{arriveTangent=0.6296, interpMode=RCIM_Cubic, leaveTangent=0.6296, tangentMode=RCTM_User, time=0.4, value=0}, <at 0.6 1>, <at 1 1>]}}})
    Index 0 Param 2=&FloatFromCurve(FloatCurve=DataInterface() {propertyValues {Curve {keys [{arriveTangent=2.1672, interpMode=RCIM_Cubic, leaveTangent=2.1672, tangentMode=RCTM_User, time=0.4, value=0}, <at 0.6 1>, <at 1 1>]}}})
    Module.Index0Float0Write=true
    Module.Index0Float1Write=true
    Module.Index0Float2Write=false
    Module.Index0Float3Write=false
   )
   &ScaleSpriteSize(
    Scale Factor=&Vector2DFromCurve(
     Vector2Curve=DataInterface() {propertyValues {
       XCurve {
        defaultValue=0
        keys [
         {arriveTangent=5.7879, interpMode=RCIM_Cubic, leaveTangent=5.7879, tangentMode=RCTM_User, time=0, value=-0.1}
         {arriveTangent=0.1448, interpMode=RCIM_Cubic, leaveTangent=0.1448, tangentMode=RCTM_User, time=0.3713, value=0.8795}
         {arriveTangent=-0.249, interpMode=RCIM_Cubic, leaveTangent=-0.249, time=0.45, value=0.5251}
         {arriveTangent=0.8972, interpMode=RCIM_Cubic, leaveTangent=0.8972, time=0.55, value=0.835}
         {arriveTangent=1.1126, arriveTangentWeight=0.2416, interpMode=RCIM_Cubic, leaveTangent=1.1126, tangentMode=RCTM_User, time=0.65, value=0.7045}
         {arriveTangent=0.0976, interpMode=RCIM_Cubic, leaveTangent=0.0976, tangentMode=RCTM_User, time=0.8, value=1.4992}
         {arriveTangent=1.3217, interpMode=RCIM_Cubic, leaveTangent=1.3217, tangentMode=RCTM_User, time=0.95, value=0.7}
         {arriveTangent=-1.8363, arriveTangentWeight=0.0918, interpMode=RCIM_Cubic, leaveTangent=-1.8363, tangentMode=RCTM_User, time=1, value=0}
        ]
       }
       YCurve {
        defaultValue=0
        keys [
         {arriveTangent=5.7879, interpMode=RCIM_Cubic, leaveTangent=5.7879, tangentMode=RCTM_User, time=0, value=0.2603}
         {arriveTangent=0.0728, interpMode=RCIM_Cubic, leaveTangent=0.0728, tangentMode=RCTM_User, time=0.3719, value=0.6079}
         {arriveTangent=-0.2262, interpMode=RCIM_Cubic, leaveTangent=-0.2262, time=0.45, value=0.9735}
         {arriveTangent=-1.3449, interpMode=RCIM_Cubic, leaveTangent=-1.3449, time=0.5475, value=0.5681}
         {arriveTangent=1.1126, arriveTangentWeight=0.2416, interpMode=RCIM_Cubic, leaveTangent=1.1126, tangentMode=RCTM_User, time=0.65, value=0.7045}
         {arriveTangent=0.929, interpMode=RCIM_Cubic, leaveTangent=0.929, time=0.8, value=1.316}
         {arriveTangent=-0.5366, interpMode=RCIM_Cubic, leaveTangent=-0.5366, tangentMode=RCTM_User, time=0.95, value=0.9832}
         {arriveTangent=-1.8363, arriveTangentWeight=0.0918, interpMode=RCIM_Cubic, leaveTangent=-1.8363, tangentMode=RCTM_User, time=1, value=0}
        ]
       }
      }}
     ScaleCurve=&MakeVector2D(X=1.3, Y=&Add_Float(A=1.3, B=&Multiply_Float(A=$System.HeartDistortion, B=1)))
    )
    Scale Sprite Size Mode=::NewEnumerator2
   ) {module=ScaleSpriteSize002}
   &SubUVAnimation(
    Direct Frame Index:NiagaraFloat=&MaskFloat(
     Mask Source=0
     Mask Parameter=?
     Mask=0
     Masked Float=&FloatFromCurve(FloatCurve=DataInterface() {propertyValues {Curve {defaultValue=1, keys [{tangentMode=RCTM_User, time=0.6, value=0}, {tangentMode=RCTM_User, time=0.7, value=1}]}}})
     Passthrough Float=0
    )
    "Sprite Renderer "=DataInterface() {propertyValues {SpriteRenderer=&self:Intro_Rebuild_0.NiagaraSpriteRendererProperties_2}}
    SubUV Animation Mode:ENiagara_SubUVLookupModeV2=::NewEnumerator5
   )
   &ScaleColor(
    Scale Alpha=&FloatFromCurve(FloatCurve=DataInterface() {propertyValues {Curve {
        defaultValue=0
        keys [
         {arriveTangent=9.0725, interpMode=RCIM_Cubic, leaveTangent=9.0725, tangentMode=RCTM_User, time=0, value=0}
         {arriveTangent=-0.2441, leaveTangent=-0.2441, time=0.1408, value=0.9847}
         {arriveTangent=-1.146, interpMode=RCIM_Cubic, leaveTangent=-1.146, tangentMode=RCTM_User, time=0.9, value=1}
         {arriveTangent=-0.0405, interpMode=RCIM_Cubic, leaveTangent=-0.0405, tangentMode=RCTM_User, time=1, value=0}
        ]
       }}})
    Scale RGB=&Multiply_VectorByFloat(
     Vector=&MakeVectorFromLinearColorRGB(Color=$System.ColorIndoor)
     Float=&FloatFromCurve(FloatCurve=DataInterface() {propertyValues {Curve {keys [
          <at 0 1 RCIM_Constant>
          <at 0.8 1 RCIM_Constant>
          <at 0.84 0.8 RCIM_Constant>
          {interpMode=RCIM_Cubic, tangentMode=RCTM_User, time=0.8753, value=2}
          <at 0.9443 1>
         ]}}})
    )
    ScaleA=true
    ScaleRGB=true
   )
   &PointForce(
    Force Origin=&AddVectorToPosition(Position=Linked() {linkedVariable {name=System.BurstPosition, type=Vector3f}}, Vector=<0 0 0>)
    Force Strength=&MaskFloat(
     Mask Source=0
     Mask Parameter=?
     Mask=0
     Masked Float=0
     Passthrough Float=&Multiply_Float(
      A=&RandomRangeFloat(Minimum=300, Maximum=7000)
      B=&Multiply_Float(
       A=$Emitter.VelocityForce
       B=&FloatFromCurve(
        FloatCurve=DataInterface() {propertyValues {Curve {keys [
            <at 0 1>
            {arriveTangentWeight=0.25, interpMode=RCIM_Cubic, leaveTangentWeight=0.0833, tangentMode=RCTM_User, time=0.78, value=1}
            <at 1 0>
            <at 1 0>
           ]}}}
        CurveIndex=$Emitter.Age
       )
      )
     )
    )
    Module.Use Falloff=false
    Module.UseLinearForce=false
   )
   &AccelerationForce(
    Acceleration=&MaskVector(
     Mask Source=0
     Mask Parameter=?
     Mask=0
     Masked Vector=<0 0 0>
     Passthrough Vector=&Multiply_VectorByFloat(Vector=<0 0 -980>, Float=$Emitter.VelocityForce)
    )
    Coordinate Space=::Simulation
   )
   &Drag(Drag=10, Use Linear Drag=true, Use Rotational Drag=false)
   &SolveForcesAndVelocity(Clamp Velocity=false, Limit Acceleration=false)
   /SetVariables(Particles.Position=/Niagara/DynamicInputs/Helpers/MaskPosition(
     Mask Source:NiagaraInt32=0
     Mask Parameter:NiagaraInt32=?
     Mask:NiagaraInt32=1
     Masked Position:NiagaraPosition=&Multiply_Vector(
      A=/Niagara/DynamicInputs/Vectors/RotateVector(
       Vector To Rotate:Vector3f=&AddVectorToPosition(
        Position=&ConvertPositionToVector(Input Position=$Particles.Position, Conversion Operation:ENiagara_LWCConvertPosToVec=::NewEnumerator0)
        Vector=&MakeVector(
         X=0
         Y=0
         Z=&Add_Float(
          A=&FloatFromCurve(
           FloatCurve=DataInterface() {propertyValues {Curve {keys [
               <at -0.0007 -0.1889 RCIM_Cubic>
               {interpMode=RCIM_Cubic, tangentMode=RCTM_User, time=0.15, value=0.1611}
               {interpMode=RCIM_Cubic, tangentMode=RCTM_User, time=0.4, value=-0.2889}
               <at 0.7 0.2111>
              ]}}}
           CurveIndex=&Add_Float(
            A=$Particles.NormalizedAge
            B=&Multiply_FloatByInt(Float=0.1, Integer=&ReturnExecIndex())
           )
          )
          B=&FloatFromCurve(
           FloatCurve=DataInterface() {propertyValues {Curve {keys [
               {interpMode=RCIM_Cubic, leaveTangent=200.8131, tangentMode=RCTM_Break, time=0.75, value=0}
               {arriveTangent=-11.7554, interpMode=RCIM_Cubic, leaveTangent=-11.7554, tangentMode=RCTM_User, time=0.8576, value=3}
               {arriveTangent=-826.6981, tangentMode=RCTM_User, time=1, value=-30}
              ]}}}
           CurveIndex=$Engine.System.Age
          )
         )
        )
       )
       Yaw:NiagaraFloat=&Multiply_Float(
        A=$Emitter.SpinIntensity
        B=&FloatFromCurve(
         FloatCurve=DataInterface() {propertyValues {Curve {keys [
             {
              arriveTangent=50.4621
              arriveTangentWeight=15.4483
              interpMode=RCIM_Cubic
              leaveTangent=50.4621
              leaveTangentWeight=15.4483
              tangentMode=RCTM_User
              tangentWeightMode=RCTWM_WeightedBoth
              time=0
              value=100
             }
             {
              arriveTangent=-0.0005
              arriveTangentWeight=0.2333
              interpMode=RCIM_Cubic
              leaveTangent=-0.0005
              leaveTangentWeight=0.0491
              tangentMode=RCTM_Break
              tangentWeightMode=RCTWM_WeightedBoth
              time=0.7
              value=220
             }
             {
              arriveTangent=-16.9689
              arriveTangentWeight=0.4829
              interpMode=RCIM_Cubic
              tangentMode=RCTM_Break
              tangentWeightMode=RCTWM_WeightedBoth
              time=0.75
              value=0
             }
            ]}}}
         CurveIndex=$Emitter.Age
        )
       )
       Pitch:NiagaraFloat=0
       Roll:NiagaraFloat=0
       Rotation Rate:NiagaraFloat=$Engine.DeltaTime
      )
      B=&VectorFromCurve(
       VectorCurve=DataInterface() {propertyValues {
         XCurve {keys [
           {interpMode=RCIM_Cubic, tangentMode=RCTM_User, time=0.18, value=0.9}
           {arriveTangentWeight=0.0116, interpMode=RCIM_Cubic, leaveTangentWeight=0.0292, tangentMode=RCTM_User, time=0.2, value=1.5}
           {interpMode=RCIM_Cubic, tangentMode=RCTM_User, time=0.25, value=0.9}
           {interpMode=RCIM_Cubic, tangentMode=RCTM_User, time=0.4, value=1.05}
           {arriveTangent=0.1667, interpMode=RCIM_Cubic, leaveTangent=0.1667, time=0.5, value=1}
           <at 0.7 1.1>
          ]}
         YCurve {keys [
           {interpMode=RCIM_Cubic, tangentMode=RCTM_User, time=0.18, value=0.9}
           {arriveTangentWeight=0.0116, interpMode=RCIM_Cubic, leaveTangentWeight=0.0292, tangentMode=RCTM_User, time=0.2, value=1.5}
           {interpMode=RCIM_Cubic, tangentMode=RCTM_User, time=0.25, value=0.9}
           {interpMode=RCIM_Cubic, tangentMode=RCTM_User, time=0.4, value=1.05}
           {arriveTangent=0.1667, interpMode=RCIM_Cubic, leaveTangent=0.1667, time=0.5, value=1}
           <at 0.7 1.1>
          ]}
         ZCurve {keys [<at 0 1>, <at 0.35 1>, {interpMode=RCIM_Cubic, tangentMode=RCTM_User, time=0.4025, value=0.96}, <at 1 1>]}
        }}
       CurveIndex=$Emitter.Age
       Scale Curve=<1 1 0.99>
      )
     )
     Passthrough Position:NiagaraPosition=$Particles.Position
    )) {isSetParameters=true}
   &self:GimbalLock_01(Gimbal Lock Direction:NiagaraPosition=&ReturnCameraVector(Camera Vectors=::NewEnumerator4, Coordinate Space=::Local, Origin Position=$Engine.Emitter.SimulationPosition)) {module=GimbalLock_01}
   /SetVariables(
    Particles.Droplet DynP 00:Vector4f=&MakeVector4(
     X=&RandomRangeFloat(Minimum=0, Maximum=1)
     Y=&FloatFromCurve(FloatCurve=DataInterface() {propertyValues {Curve {keys [<at 0.75 1.1>,<at 0.8 0.9>]}}}, CurveIndex=$Emitter.Age)
     Z=0
     W=0
    )
    Particles.DropletSize:Vector2f=&Multiply_Vector2D(A=$Particles.BobaSize, B=<1.5 3.0143>)
   ) {isSetParameters=true}
   /SetVariables(
    Particles.Boba DynP 00:Vector4f=&MakeVector4(
     X=&FloatFromCurve(FloatCurve=DataInterface() {propertyValues {Curve {keys [<at 0.75 0>,<at 0.8 1>]}}})
     Y=&FloatFromCurve(FloatCurve=DataInterface() {propertyValues {Curve {keys [<at 0.7 0.5>,<at 0.75 1>]}}})
     Z=1
     W=1
    )
    Particles.Boba DynP 01:Vector4f=&MakeVector4(X=0, Y=0, Z=0, W=0)
    Particles.BobaColor:LinearColor=&LerpLinearColors(
     StartColor=&LerpLinearColors(
      StartColor=<rgba 1 0.5302 0.9204 1>
      EndColor=<rgba 0.2448 0.0805 0 1>
      LerpFactor=&ClampFloat(
       Float=&Divide_Float(
        A=&Subtract_Float(
         A=&Subtract_Float(
          A=&MakeFloatFromVector(
           Vector=&ConvertPositionToVector(Input Position=$Particles.Position, Conversion Operation:ENiagara_LWCConvertPosToVec=::NewEnumerator0)
           Channel=::NewEnumerator0
          )
          B=&MakeFloatFromVector(
           Vector=&ConvertPositionToVector(Input Position=$Engine.Emitter.SimulationPosition, Conversion Operation:ENiagara_LWCConvertPosToVec=::NewEnumerator0)
           Channel=::NewEnumerator0
          )
         )
         B=1
        )
        B=2
       )
       Max=0.5
      )
     )
     EndColor=<rgba 0.3073 0.2456 0.2154 1>
     LerpFactor=&FloatFromCurve(
      FloatCurve=DataInterface() {propertyValues {Curve {defaultValue=1, keys [{tangentMode=RCTM_User, time=0.77, value=0}, {tangentMode=RCTM_User, time=0.8, value=1}]}}}
      CurveIndex=$Emitter.Age
     )
    )
    Particles.BobaSize:Vector2f=&Multiply_Vector2D_ByFloat(
     Float=&FloatFromCurve(FloatCurve=DataInterface() {propertyValues {Curve {keys [
          {interpMode=RCIM_Cubic, tangentMode=RCTM_User, time=0, value=0}
          <at 0.2 1>
          {interpMode=RCIM_Cubic, tangentMode=RCTM_User, time=0.9, value=1}
          <at 1 0>
         ]}}})
     Vector2D=$Particles.Initial.BobaSize
    )
   ) {isSetParameters=true}
   /SetVariables(Particles.HeartBlendColor:LinearColor=&MultiplyLinearColor(
     Linear Color=&MultiplyLinearColor(
      Linear Color=$Particles.Initial.Color
      Scale Factor=&MakeVector4_FromLinearColor(LinearColor=$System.ColorIndoor)
     )
     Scale Factor=&MakeVector4(X=1.5, Y=1.5, Z=1.5, W=5)
    )) {isSetParameters=true}
   /SetVariables(Particles.GimbalPos:NiagaraPosition=&AddVectorToPosition(
     Position=$Particles.Position
     Vector=&Lerp_Vector(
      A=$System.HeartIntroSpawnPosition
      B=$System.BurstPosition
      Alpha=&MaskFloat(
       Mask Source=0
       Mask Parameter=?
       Mask=0
       Masked Float=&FloatFromCurve(
        FloatCurve=DataInterface() {propertyValues {Curve {keys [
            {
             arriveTangent=7.872
             arriveTangentWeight=0.6727
             interpMode=RCIM_Cubic
             leaveTangent=7.872
             leaveTangentWeight=0.6727
             tangentMode=RCTM_User
             tangentWeightMode=RCTWM_WeightedBoth
             time=0
             value=0
            }
            {
             arriveTangent=0.0441
             arriveTangentWeight=0.6975
             interpMode=RCIM_Cubic
             leaveTangent=0.0441
             leaveTangentWeight=0.6975
             tangentMode=RCTM_User
             tangentWeightMode=RCTWM_WeightedBoth
             time=0.7
             value=1
            }
           ]}}}
        CurveIndex=$Emitter.Age
       )
       Passthrough Float=&FloatFromCurve(
        FloatCurve=DataInterface() {propertyValues {Curve {keys [
            {
             arriveTangent=7.872
             arriveTangentWeight=0.6727
             interpMode=RCIM_Cubic
             leaveTangent=7.872
             leaveTangentWeight=0.6727
             tangentMode=RCTM_User
             tangentWeightMode=RCTWM_WeightedBoth
             time=0
             value=0
            }
            {
             arriveTangent=0.0441
             arriveTangentWeight=0.638
             interpMode=RCIM_Cubic
             leaveTangent=0.0441
             leaveTangentWeight=0.638
             tangentMode=RCTM_User
             tangentWeightMode=RCTWM_WeightedBoth
             time=0.7
             value=1
            }
           ]}}}
        CurveIndex=&Add_Float(
         A=$Particles.NormalizedAge
         B=&MaskFloat(
          Mask Source=0
          Mask Parameter=?
          Mask=1
          Masked Float=&RandomRangeFloat(Minimum=-0.02, Maximum=0.02)
          Passthrough Float=0
         )
        )
       )
      )
     )
    )) {isSetParameters=true}
  ]
  rendererClasses [&NiagaraSpriteRendererProperties]
  renderers [
   {rendererClass=&NiagaraSpriteRendererProperties}
   {rendererClass=&NiagaraSpriteRendererProperties, rendererIndex=1}
   {rendererClass=&NiagaraSpriteRendererProperties, rendererIndex=2}
   {rendererClass=&NiagaraSpriteRendererProperties, rendererIndex=3}
   {rendererClass=&NiagaraSpriteRendererProperties, rendererIndex=4}
  ]
 }
 Pop {
  emitterUpdateScript [
   &EmitterState()
   &SpawnBurst_Instantaneous(
    Spawn Count=3
    Spawn Time=&Subtract_Float(A=$System.IntroLength, B=0)
    Use Spawn Probability=false
    UseLoopCountLimit=false
   )
   &SpawnBurst_Instantaneous(
    Spawn Count=3
    Spawn Group=1
    Spawn Time=&Subtract_Float(A=$System.IntroLength, B=0)
    Use Spawn Probability=false
    UseLoopCountLimit=false
   ) {module=SpawnBurst_Instantaneous002}
   &SpawnBurst_Instantaneous(
    Spawn Count=15
    Spawn Group=2
    Spawn Time=&Subtract_Float(A=$System.IntroLength, B=0)
    Use Spawn Probability=false
    UseLoopCountLimit=false
   ) {module=SpawnBurst_Instantaneous001}
  ]
  particleSpawnScript [
   &PartitionParticles()
   /SetVariables(Particles.VisibilityTag=&SpawnGroup()) {isSetParameters=true}
   &InitializeParticle(
    Color=<rgba 0.491 0.491 0.491 1>
    Lifetime=&MaskFloat(
     Mask Source=0
     Mask Parameter=?
     Mask=2
     Masked Float=&RandomRangeFloat(Minimum=0.4, Maximum=0.7)
     Passthrough Float=&RandomRangeFloat(Minimum=0.25, Maximum=0.5)
    )
    Position Offset=$System.BurstPosition
    Sprite Rotation Mode=::NewEnumerator2
    Sprite Size Mode=::NewEnumerator3
    Uniform Sprite Size=&MaskFloat(
     Mask Source=0
     Mask Parameter=?
     Mask=2
     Masked Float=&RandomRangeFloat(Minimum=1, Maximum=1.5)
     Passthrough Float=&RandomRangeFloat(Minimum=8, Maximum=15)
    )
    UsePositionOffset=false
    Write Material Random=true
   )
   /SetVariables(Particles.DropletSize:Vector2f=&Multiply_Vector2D(A=$Particles.SpriteSize, B=<1 3>)) {isSetParameters=true}
   &DynamicMaterialParameters(
    Index 0 Param 4=1
    Module.Index0Float0Write=true
    Module.Index0Float1Write=true
    Module.Index0Float2Write=true
    Module.Index0Float3Write=true
   )
   &SubUVAnimation(
    Evaluation Type=::NewEnumerator0
    "Sprite Renderer "=DataInterface() {propertyValues {SpriteRenderer=&self:Pop_0.NiagaraSpriteRendererProperties_1}}
    SubUV Animation Mode:ENiagara_SubUVLookupModeV2=::NewEnumerator2
    UseEndFrame=false
    UseStartFrame=false
   )
   &ShapeLocation(
    MaskSpawnGroup=false
    Offset=<0 0 1>
    Offset Coordinate Space=::Local
    Offset Mode=::NewEnumerator0
    Override Seed=false
    Shape Origin:Vector3f=$System.BurstPosition
    Sphere Radius=5
    UseOwnerScale=true
   )
   &AddVelocity(
    Module.Use Falloff Distance=false
    Origin Offset=<0 0 -5>
    Override Seed=false
    Velocity Mode=::NewEnumerator1
    Velocity Origin:Vector3f=$System.BurstPosition
    Velocity Speed=&MaskFloat(
     Mask Source=0
     Mask Parameter=?
     Mask=2
     Masked Float=&RandomRangeFloat(Minimum=150, Maximum=300)
     Passthrough Float=&RandomRangeFloat(Minimum=75, Maximum=100)
    )
   )
  ]
  particleUpdateScript [
   &ParticleState()
   &ScaleSpriteSize(
    Scale Sprite Size Mode=::NewEnumerator0
    Uniform Scale Factor=&FloatFromCurve(FloatCurve=DataInterface() {propertyValues {Curve {keys [
         {arriveTangent=5.7724, interpMode=RCIM_Cubic, leaveTangent=5.7724, tangentMode=RCTM_User, time=0, value=1}
         {arriveTangent=0.788, tangentMode=RCTM_User, time=1, value=4}
        ]}}})
   ) {module=ScaleSpriteSize002}
   &ScaleColor(
    Scale Alpha=&FloatFromCurve(FloatCurve=DataInterface() {propertyValues {Curve {
        defaultValue=0
        keys [
         {arriveTangent=0.3137, interpMode=RCIM_Cubic, leaveTangent=0.3137, tangentMode=RCTM_User, time=0, value=1}
         {arriveTangent=-2.0958, interpMode=RCIM_Cubic, leaveTangent=-2.0958, tangentMode=RCTM_User, time=1, value=0}
        ]
       }}})
    Scale RGB=&VectorFromCurve(VectorCurve=DataInterface() {propertyValues {
       XCurve {keys [
         <at 0 0 RCIM_Cubic>
         {interpMode=RCIM_Cubic, tangentMode=RCTM_User, time=0.2, value=2}
         {
          arriveTangent=0.0281
          arriveTangentWeight=0.6561
          interpMode=RCIM_Cubic
          leaveTangent=0.0281
          leaveTangentWeight=0.6561
          tangentMode=RCTM_User
          tangentWeightMode=RCTWM_WeightedBoth
          time=1
          value=0.3
         }
        ]}
       YCurve {keys [
         <at 0 0 RCIM_Cubic>
         {interpMode=RCIM_Cubic, tangentMode=RCTM_User, time=0.2, value=2}
         {
          arriveTangent=0.0281
          arriveTangentWeight=0.6561
          interpMode=RCIM_Cubic
          leaveTangent=0.0281
          leaveTangentWeight=0.6561
          tangentMode=RCTM_User
          tangentWeightMode=RCTWM_WeightedBoth
          time=1
          value=0.3
         }
        ]}
       ZCurve {keys [
         <at 0 0 RCIM_Cubic>
         {interpMode=RCIM_Cubic, tangentMode=RCTM_User, time=0.2, value=2}
         {
          arriveTangent=0.0281
          arriveTangentWeight=0.6561
          interpMode=RCIM_Cubic
          leaveTangent=0.0281
          leaveTangentWeight=0.6561
          tangentMode=RCTM_User
          tangentWeightMode=RCTWM_WeightedBoth
          time=1
          value=0.3
         }
        ]}
      }})
    ScaleA=true
    ScaleRGB=true
   )
   &AccelerationForce(Acceleration=&MaskVectorBySpawnGroup(Masked Vector=<0 0 -980>, Passthrough Vector=<1 1 -490>, Spawn Group Mask=2))
   &SolveForcesAndVelocity(Clamp Velocity=false, Limit Acceleration=false)
   &ScaleSpriteSizeBySpeed(Max Scale Factor=<2 5>)
   /SetVariables(Particles.Glow Color=&Multiply_LinearColorByFloat(
     Linear Color=&MultiplyLinearColor(
      Linear Color=$Particles.Color
      Scale Factor=&MakeVector4_FromLinearColor(LinearColor=<rgba 1 0.5877 0.8563 1>)
     )
     Float=0.33
    )) {isSetParameters=true}
   &DynamicMaterialParameters(
    Index 0 Param 2=&FloatFromCurve(FloatCurve=DataInterface() {propertyValues {Curve {defaultValue=1, keys [{interpMode=RCIM_Cubic, tangentMode=RCTM_User, time=0, value=0}, {tangentMode=RCTM_User, time=1, value=1}]}}})
    Index 0 Param 3=&FloatFromCurve(FloatCurve=DataInterface() {propertyValues {Curve {keys [{interpMode=RCIM_Cubic, tangentMode=RCTM_User, time=0, value=0.3749}, <at 1 0>]}}})
    Module.Index0Float0Write=true
    Module.Index0Float1Write=true
    Module.Index0Float2Write=true
    Module.Index0Float3Write=true
   ) {module=DynamicMaterialParameters001}
  ]
  rendererClasses [&NiagaraSpriteRendererProperties]
  renderers [
   {rendererClass=&NiagaraSpriteRendererProperties}
   {rendererClass=&NiagaraSpriteRendererProperties, rendererIndex=1}
   {rendererClass=&NiagaraSpriteRendererProperties, rendererIndex=2}
  ]
 }
 Spawn {
  emitterUpdateScript [
   &EmitterState()
   &SpawnBurst_Instantaneous(Spawn Count=11, Use Spawn Probability=false, UseLoopCountLimit=false)
  ]
  particleSpawnScript [
   &InitializeParticle(
    Color=<rgba 0.8 0.8 0.8 1>
    Lifetime=&RandomRangeFloat(Minimum=0.4, Maximum=0.5)
    Sprite Size Mode=::NewEnumerator3
    Uniform Sprite Size=&RandomRangeFloat(Minimum=1, Maximum=2)
    UsePositionOffset=false
    Write Material Random=true
   )
   &ShapeLocation(
    MaskSpawnGroup=false
    Non Uniform Scale=<1 1 5>
    Offset=$System.HeartIntroSpawnPosition
    Offset Coordinate Space=::Local
    Offset Mode=::NewEnumerator0
    Override Seed=false
    Sphere Radius=0.5
    UseOwnerScale=true
   )
   &AddVelocity(
    Cone Angle=23
    Cone Axis=<0 0 1>
    Module.UseConeFalloff=false
    Module.UseVelDistribution=false
    Override Seed=false
    Velocity Mode=::NewEnumerator2
    Velocity Speed=&RandomRangeFloat(Minimum=150, Maximum=300)
   )
   &ApplyInitialForces()
   &CameraOffset(Camera Offset Amount=5)
   &SubUVAnimation(Evaluation Type=::NewEnumerator0, SubUV Animation Mode:ENiagara_SubUVLookupModeV2=::NewEnumerator2, UseEndFrame=false, UseStartFrame=false)
  ]
  particleUpdateScript [
   &AccelerationForce(Acceleration=<0 0 -980>)
   &SolveForcesAndVelocity(Clamp Velocity=false, Limit Acceleration=false)
   &ParticleState()
   &ScaleSpriteSizeBySpeed(Max Scale Factor=<2 20>)
   &DynamicMaterialParameters(
    Index 0 Param 2=&FloatFromCurve(FloatCurve=DataInterface() {propertyValues {Curve {defaultValue=1, keys [{interpMode=RCIM_Cubic, tangentMode=RCTM_User, time=0, value=0}, {tangentMode=RCTM_User, time=1, value=1}]}}})
    Index 0 Param 3=&FloatFromCurve(FloatCurve=DataInterface() {propertyValues {Curve {keys [{interpMode=RCIM_Cubic, tangentMode=RCTM_User, time=0, value=0.3749}, <at 1 0>]}}})
    Module.Index0Float0Write=true
    Module.Index0Float1Write=true
    Module.Index0Float2Write=false
    Module.Index0Float3Write=true
   ) {module=DynamicMaterialParameters001}
  ]
  rendererClasses [&NiagaraSpriteRendererProperties]
  renderers [{rendererClass=&NiagaraSpriteRendererProperties}]
 }
}
summary {
 systemName=NS_ReactFX_TapiTime
 emitters [
  {emitterName=Pop, rendererClasses [&NiagaraSpriteRendererProperties]}
  {emitterName=Spawn, rendererClasses [&NiagaraSpriteRendererProperties]}
  {emitterName=Intro_Rebuild, rendererClasses [&NiagaraSpriteRendererProperties]}
 ]
 userVariables [
  User.bIsFrontEnd {type=NiagaraBool}
  User.bIsFrontEndPreview {type=NiagaraBool}
 ]
}
