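# Niagara reaction effect for the TeaPup cosmetic companion, with an intro sprite and a burst of hearts, flare and spikes.
# Intro plays a petting-reaction sound and animates a jittering, scaling glow sprite for System.IntroLength around System.BurstPosition.
# HeartsFlare packs the bursts and spike stream into one emitter, using spawn groups, a VisibilityTag and four sprite renderers, with time-of-day tinting.
# Hearts001, Impact_Spikes and NE_First_Flare are disabled standalone versions of those layers.
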NiagaraSystem /CosmeticCompanions/FX/React/NS_ReactFX_TeaPup
dependencies {
 usedDataInterfaces [
  &NiagaraDataInterfaceCurve
  &NiagaraDataInterfaceColorCurve
  &NiagaraDataInterfaceVector2DCurve
  &NiagaraDataInterfaceAudioPlayer
 ]
 usedDynamicInputs [/Niagara/DynamicInputs/TypeConversions/MakeCustomPositionFromBool, &FN_FrontEndVector3Modifier]
 usedModules [
  &self:SystemSpawnScript.SystemScriptSource.SystemScriptGraph.NiagaraNodeAssignment_8.SetVariables_#1
  &self:SystemSpawnScript.SystemScriptSource.SystemScriptGraph.NiagaraNodeAssignment_0.SetVariables_#2
  &self:SystemSpawnScript.SystemScriptSource.SystemScriptGraph.NiagaraNodeAssignment_9.SetVariables_#3
  &self:SystemSpawnScript.SystemScriptSource.SystemScriptGraph.NiagaraNodeAssignment_1.SetVariables_#4
  &self:SystemSpawnScript.SystemScriptSource.SystemScriptGraph.NiagaraNodeAssignment_10.SetVariables_#5
  &self:SystemSpawnScript.SystemScriptSource.SystemScriptGraph.NiagaraNodeAssignment_11.SetVariables_#6
  &self:SystemSpawnScript.SystemScriptSource.SystemScriptGraph.NiagaraNodeAssignment_12.SetVariables_#7
  &self:SystemSpawnScript.SystemScriptSource.SystemScriptGraph.NiagaraNodeAssignment_13.SetVariables_#8
  &self:SystemSpawnScript.SystemScriptSource.SystemScriptGraph.NiagaraNodeAssignment_14.SetVariables_#9
  &SystemState
 ]
}
emitters {
 Hearts001 {
  bEnabled=false
  emitterUpdateScript [
   &EmitterState()
   &SpawnBurst_Instantaneous(
    Spawn Count=5
    Spawn Time=&Subtract_Float(A=$System.IntroLength, B=0)
    Use Spawn Probability=false
    UseLoopCountLimit=false
   )
   &SpawnBurst_Instantaneous(Spawn Count=5, Spawn Group=1, Spawn Time=0.8, Use Spawn Probability=false, UseLoopCountLimit=false) {module=SpawnBurst_Instantaneous002}
  ]
  particleSpawnScript [
   /SetVariables(Particles.VisibilityTag=&SpawnGroup()) {isSetParameters=true, module=SetVariables_#10}
   &InitializeParticle(
    Color=&RandomRangeLinearColor(
     Minimum=&FNTimeOfDayColor(
      Night Color=<rgba 0.3142 0.192 0.0773 1>
      Morning Color=<rgba 1 0.4843 0 1>
      Day Color=<rgba 1 0.4843 0 1>
      Evening Color=<rgba 1 0.4843 0 1>
     )
     Maximum=&FNTimeOfDayColor(
      Night Color=<rgba 0.3605 0.1368 0.0887 1>
      Morning Color=<rgba 0.9219 0.1632 0 1>
      Day Color=<rgba 0.9219 0.1632 0 1>
      Evening Color=<rgba 0.9219 0.1632 0 1>
     )
    )
    Lifetime=&MergeHelper_FloatSwitch(
     0=&RandomRangeFloat(Minimum=0.7, Maximum=0.5)
     1=&RandomRangeFloat(Minimum=0.3, Maximum=0.5)
     2=&RandomRangeFloat(Minimum=0.3, Maximum=0.5)
     3=&RandomRangeFloat(Minimum=0.4, Maximum=0.6)
     Index=&SpawnGroup()
     Num=::NewEnumerator0
    )
    Position Offset=$System.BurstPosition
    Sprite Rotation Mode=::NewEnumerator0
    Sprite Size Mode=::NewEnumerator3
    Uniform Sprite Size=&MergeHelper_FloatSwitch(
     0=&Multiply_Float(A=1, B=&RandomRangeFloat(Minimum=4, Maximum=16))
     1=&RandomRangeFloat(Minimum=8, Maximum=22)
     2=&RandomRangeFloat(Minimum=8, Maximum=22)
     3=&RandomRangeFloat(Minimum=8, Maximum=15)
     Index=&SpawnGroup()
     Num=::NewEnumerator0
    )
    UsePositionOffset=true
    Write Material Random=true
   )
   &ShapeLocation(MaskSpawnGroup=false, Offset=<0 0 2>, Offset Mode=::NewEnumerator0, Override Seed=false, Sphere Radius=10, UseOwnerScale=true)
   &AddVelocity(
    Constrain To Radius=100
    Module.Use Falloff Distance=false
    Origin Offset=<0 5 -10>
    Override Seed=false
    Velocity Mode=::NewEnumerator1
    Velocity Origin:Vector3f=$System.BurstPosition
    Velocity Speed=&MergeHelper_FloatSwitch(0=100, 1=50, 2=?, 3=?, Index=&SpawnGroup(), Num=::NewEnumerator0)
   )
  ]
  particleUpdateScript [
   &ParticleState()
   &SpriteRotationRate(Rotation Rate=&RandomRangeFloat(Minimum=-100, Maximum=100))
   &ScaleSpriteSize(Uniform Curve Sprite Scale=DataInterface() {propertyValues {Curve {
       defaultValue=1
       keys [
        {arriveTangent=4.1399, interpMode=RCIM_Cubic, leaveTangent=4.1399, tangentMode=RCTM_User, time=0, value=0.5}
        {arriveTangent=0.0224, interpMode=RCIM_Cubic, leaveTangent=0.0224, tangentMode=RCTM_User, time=1, value=1}
       ]
      }}})
   /SetVariables(Particles.Flicker:NiagaraFloat=&Waveform(
     "Waveform [1]"=::NewEnumerator0
     "[1] Frequency"=&RandomRangeFloat(Minimum=3, Maximum=5)
     "[1] Amplitude Min/Max"=<0 1>
    )) {isSetParameters=true, module=SetVariables_#11}
   &Drag(
    Drag=&MergeHelper_FloatSwitch(
     0=&Multiply_Float(
      A=&RandomRangeFloat(Minimum=3, Maximum=7)
      B=&FloatFromCurve(
       FloatCurve=DataInterface() {propertyValues {Curve {defaultValue=1, keys [{tangentMode=RCTM_User, time=0.2, value=0}, {tangentMode=RCTM_User, time=0.35, value=2}]}}}
       Scale Curve=&RandomRangeFloat(Minimum=3, Maximum=1)
      )
     )
     1=&RandomRangeFloat(Minimum=3, Maximum=1)
     2=?
     3=?
     Index=&SpawnGroup()
     Num=::NewEnumerator0
    )
    Use Linear Drag=true
    Use Rotational Drag=true
   )
   &SolveForcesAndVelocity(Clamp Velocity=false, Limit Acceleration=false)
   &ScaleColor(
    Scale Alpha=&MergeHelper_FloatSwitch(
     0=&FloatFromCurve(FloatCurve=DataInterface() {propertyValues {Curve {
         defaultValue=1
         keys [
          {arriveTangent=-0.2867, interpMode=RCIM_Cubic, leaveTangent=-0.2867, tangentMode=RCTM_User, time=0, value=0}
          {arriveTangent=-3.8875, interpMode=RCIM_Cubic, leaveTangent=-3.8875, tangentMode=RCTM_User, time=0.3695, value=1}
          {arriveTangent=-0.0405, interpMode=RCIM_Cubic, leaveTangent=-0.0405, tangentMode=RCTM_User, time=0.8, value=-0.0266}
         ]
        }}})
     1=&FloatFromCurve(FloatCurve=DataInterface() {propertyValues {Curve {
         defaultValue=1
         keys [
          {arriveTangent=-0.2867, interpMode=RCIM_Cubic, leaveTangent=-0.2867, tangentMode=RCTM_User, time=0, value=0}
          {arriveTangent=-3.8875, interpMode=RCIM_Cubic, leaveTangent=-3.8875, tangentMode=RCTM_User, time=0.3695, value=1}
          {arriveTangent=-0.0405, interpMode=RCIM_Cubic, leaveTangent=-0.0405, tangentMode=RCTM_User, time=0.8, value=-0.0266}
         ]
        }}})
     2=?
     3=?
     Index=&SpawnGroup()
     Num=::NewEnumerator0
    )
    ScaleA=true
    ScaleRGB=false
   )
   /SetVariables(
    Particles.Hearts SpriteSize:Vector2f=&Vector2DFromFloat(Value=&FloatFromCurve(
      FloatCurve=DataInterface() {propertyValues {Curve {
         defaultValue=1
         keys [
          {arriveTangent=3.2566, interpMode=RCIM_Cubic, leaveTangent=3.2566, tangentMode=RCTM_User, time=0, value=0.5}
          {arriveTangent=-0.5, interpMode=RCIM_Cubic, leaveTangent=-0.5, time=0.5039, value=1.1092}
          {arriveTangent=0.0224, interpMode=RCIM_Cubic, leaveTangent=0.0224, tangentMode=RCTM_User, time=1, value=0}
         ]
        }}}
      Scale Curve=&Multiply_Float(A=1, B=&RandomRangeFloat(Minimum=4, Maximum=16))
     ))
    Particles.Small hearts SpriteSize:Vector2f=&Vector2DFromFloat(Value=&FloatFromCurve(
      FloatCurve=DataInterface() {propertyValues {Curve {
         defaultValue=1
         keys [
          {arriveTangent=1.914, interpMode=RCIM_Cubic, leaveTangent=1.914, tangentMode=RCTM_User, time=0, value=0.5}
          {arriveTangent=-0.1241, interpMode=RCIM_Cubic, leaveTangent=-0.1241, tangentMode=RCTM_User, time=1, value=1}
         ]
        }}}
      Scale Curve=&Multiply_Float(A=1, B=&RandomRangeFloat(Minimum=3, Maximum=8))
     ))
   ) {isSetParameters=true, module=SetVariables_#12}
  ]
  rendererClasses [&NiagaraSpriteRendererProperties]
  renderers [
   {rendererClass=&NiagaraSpriteRendererProperties}
   {rendererClass=&NiagaraSpriteRendererProperties, rendererIndex=1}
  ]
 }
 HeartsFlare {
  emitterUpdateScript [
   &EmitterState()
   &SpawnBurst_Instantaneous(
    Spawn Count=5
    Spawn Time=&Subtract_Float(A=$System.IntroLength, B=0)
    Use Spawn Probability=false
    UseLoopCountLimit=false
   )
   &SpawnBurst_Instantaneous(Spawn Count=5, Spawn Group=1, Spawn Time=0.8, Use Spawn Probability=false, UseLoopCountLimit=false) {module=SpawnBurst_Instantaneous002}
   &SpawnBurst_Instantaneous(Spawn Group=2, Spawn Time=$System.IntroLength, Use Spawn Probability=false, UseLoopCountLimit=false) {module=SpawnBurst_Instantaneous001}
   &SpawnRate(
    SpawnGroup=3
    SpawnRate=&FloatFromCurve(
     FloatCurve=DataInterface() {propertyValues {Curve {
        defaultValue=1
        keys [
         {arriveTangent=10.7524, interpMode=RCIM_Cubic, leaveTangent=10.7524, tangentMode=RCTM_User, time=0, value=0}
         <at 0.2657 1.0002 RCIM_Cubic>
         {arriveTangent=-10.45, interpMode=RCIM_Cubic, leaveTangent=-10.45, tangentMode=RCTM_User, time=0.5769, value=0}
        ]
       }}}
     CurveIndex=$Emitter.NormalizedLoopAge
     Scale Curve=15
    )
    Use Spawn Probability=false
   )
  ]
  particleSpawnScript [
   /SetVariables(Particles.VisibilityTag=&SpawnGroup()) {isSetParameters=true, module=SetVariables_#10}
   &InitializeParticle(
    Color=&RandomRangeLinearColor(
     Minimum=&FNTimeOfDayColor(
      Night Color=<rgba 0.3142 0.192 0.0773 1>
      Morning Color=<rgba 1 0.4843 0 1>
      Day Color=<rgba 1 0.4843 0 1>
      Evening Color=<rgba 1 0.4843 0 1>
     )
     Maximum=&FNTimeOfDayColor(
      Night Color=<rgba 0.3605 0.1368 0.0887 1>
      Morning Color=<rgba 0.9219 0.1632 0 1>
      Day Color=<rgba 0.9219 0.1632 0 1>
      Evening Color=<rgba 0.9219 0.1632 0 1>
     )
    )
    Lifetime=&MergeHelper_FloatSwitch(
     0=&RandomRangeFloat(Minimum=0.7, Maximum=0.5)
     1=&RandomRangeFloat(Minimum=0.3, Maximum=0.5)
     2=0.4
     3=&RandomRangeFloat(Minimum=0.2, Maximum=0.15)
     Index=&SpawnGroup()
     Num=::NewEnumerator2
    )
    Position Offset=$System.BurstPosition
    Sprite Rotation Mode=::NewEnumerator0
    Sprite Size Mode=::NewEnumerator3
    Uniform Sprite Size=&MergeHelper_FloatSwitch(
     0=&Multiply_Float(A=1, B=&RandomRangeFloat(Minimum=4, Maximum=16))
     1=&RandomRangeFloat(Minimum=8, Maximum=22)
     2=100
     3=&RandomRangeFloat(Minimum=8, Maximum=15)
     Index=&SpawnGroup()
     Num=::NewEnumerator1
    )
    UsePositionOffset=true
    Write Material Random=true
   )
   &ShapeLocation(
    MaskSpawnGroup=false
    Offset=&MergeHelper_VectorSwitch(0=<0 0 2>, 1=<0 0 2>, 2=<0 0 0>, 3=?, Index=&SpawnGroup(), Num=::NewEnumerator1)
    Offset Mode=::NewEnumerator0
    Override Seed=false
    Sphere Radius=&MergeHelper_FloatSwitch(0=10, 1=10, 2=0, 3=?, Index=&SpawnGroup(), Num=::NewEnumerator1)
    UseOwnerScale=true
   )
   &AddVelocity(
    Constrain To Radius=100
    Module.Use Falloff Distance=false
    Origin Offset=&MergeHelper_VectorSwitch(0=<0 5 -10>, 1=<0 5 -10>, 2=<0 0 0>, 3=<0 0 0>, Index=&SpawnGroup(), Num=::NewEnumerator2)
    Override Seed=false
    Velocity Mode=::NewEnumerator1
    Velocity Origin:Vector3f=$System.BurstPosition
    Velocity Speed=&MergeHelper_FloatSwitch(
     0=100
     1=50
     2=0
     3=&RandomRangeFloat(Minimum=200, Maximum=200)
     Index=&SpawnGroup()
     Num=::NewEnumerator2
    )
   )
   &CameraOffset(Camera Offset Amount=&MergeHelper_FloatSwitch(0=0, 1=0, 2=5, 3=?, Index=&SpawnGroup(), Num=::NewEnumerator1))
   /SetVariables(
    Particles.ImpactSize:Vector2f=&Multiply_Vector2D_ByFloat(Float=0.6, Vector2D=&RandomRangeVector2D(Minimum=<20 30>, Maximum=<10 20>))
    Particles.ImpactSubImageIndex:NiagaraFloat=&Multiply_FloatByInt(Float=1, Integer=&RandomRangeInt(Minimum=0, Maximum=3))
   ) {isSetParameters=true}
  ]
  particleUpdateScript [
   &ParticleState()
   &SpriteRotationRate(Rotation Rate=&MergeHelper_FloatSwitch(
     0=&RandomRangeFloat(Minimum=-100, Maximum=100)
     1=&RandomRangeFloat(Minimum=-100, Maximum=100)
     2=0
     3=?
     Index=&SpawnGroup()
     Num=::NewEnumerator1
    ))
   &ScaleSpriteSize(Uniform Curve Sprite Scale=DataInterface() {propertyValues {Curve {
       defaultValue=1
       keys [
        {arriveTangent=4.1399, interpMode=RCIM_Cubic, leaveTangent=4.1399, tangentMode=RCTM_User, time=0, value=0.5}
        {arriveTangent=0.0224, interpMode=RCIM_Cubic, leaveTangent=0.0224, tangentMode=RCTM_User, time=1, value=1}
       ]
      }}})
   /SetVariables(Particles.Flicker:NiagaraFloat=&Waveform(
     "Waveform [1]"=::NewEnumerator0
     "[1] Frequency"=&RandomRangeFloat(Minimum=3, Maximum=5)
     "[1] Amplitude Min/Max"=<0 1>
    )) {isSetParameters=true, module=SetVariables_#11}
   &Drag(
    Drag=&MergeHelper_FloatSwitch(
     0=&Multiply_Float(
      A=&RandomRangeFloat(Minimum=3, Maximum=7)
      B=&FloatFromCurve(
       FloatCurve=DataInterface() {propertyValues {Curve {defaultValue=1, keys [{tangentMode=RCTM_User, time=0.2, value=0}, {tangentMode=RCTM_User, time=0.35, value=2}]}}}
       Scale Curve=&RandomRangeFloat(Minimum=3, Maximum=1)
      )
     )
     1=&RandomRangeFloat(Minimum=3, Maximum=1)
     2=0
     3=&RandomRangeFloat(Minimum=200, Maximum=200)
     Index=&SpawnGroup()
     Num=::NewEnumerator2
    )
    Use Linear Drag=true
    Use Rotational Drag=true
   )
   &SolveForcesAndVelocity(Clamp Velocity=false, Limit Acceleration=false)
   &ScaleColor(
    Scale Alpha=&MergeHelper_FloatSwitch(
     0=&FloatFromCurve(FloatCurve=DataInterface() {propertyValues {Curve {
         defaultValue=1
         keys [
          {arriveTangent=-0.2867, interpMode=RCIM_Cubic, leaveTangent=-0.2867, tangentMode=RCTM_User, time=0, value=0}
          {arriveTangent=-3.8875, interpMode=RCIM_Cubic, leaveTangent=-3.8875, tangentMode=RCTM_User, time=0.3695, value=1}
          {arriveTangent=-0.0405, interpMode=RCIM_Cubic, leaveTangent=-0.0405, tangentMode=RCTM_User, time=0.8, value=-0.0266}
         ]
        }}})
     1=&FloatFromCurve(FloatCurve=DataInterface() {propertyValues {Curve {
         defaultValue=1
         keys [
          {arriveTangent=-0.2867, interpMode=RCIM_Cubic, leaveTangent=-0.2867, tangentMode=RCTM_User, time=0, value=0}
          {arriveTangent=-3.8875, interpMode=RCIM_Cubic, leaveTangent=-3.8875, tangentMode=RCTM_User, time=0.3695, value=1}
          {arriveTangent=-0.0405, interpMode=RCIM_Cubic, leaveTangent=-0.0405, tangentMode=RCTM_User, time=0.8, value=-0.0266}
         ]
        }}})
     2=?
     3=?
     Index=&SpawnGroup()
     Num=::NewEnumerator0
    )
    ScaleA=true
    ScaleRGB=false
   )
   /SetVariables(
    Particles.Hearts SpriteSize:Vector2f=&Vector2DFromFloat(Value=&FloatFromCurve(
      FloatCurve=DataInterface() {propertyValues {Curve {
         defaultValue=1
         keys [
          {arriveTangent=3.2566, interpMode=RCIM_Cubic, leaveTangent=3.2566, tangentMode=RCTM_User, time=0, value=0.5}
          {arriveTangent=-0.5, interpMode=RCIM_Cubic, leaveTangent=-0.5, time=0.5039, value=1.1092}
          {arriveTangent=0.0224, interpMode=RCIM_Cubic, leaveTangent=0.0224, tangentMode=RCTM_User, time=1, value=0}
         ]
        }}}
      Scale Curve=&Multiply_Float(A=1, B=&RandomRangeFloat(Minimum=4, Maximum=16))
     ))
    Particles.Small hearts SpriteSize:Vector2f=&Vector2DFromFloat(Value=&FloatFromCurve(
      FloatCurve=DataInterface() {propertyValues {Curve {
         defaultValue=1
         keys [
          {arriveTangent=1.914, interpMode=RCIM_Cubic, leaveTangent=1.914, tangentMode=RCTM_User, time=0, value=0.5}
          {arriveTangent=-0.1241, interpMode=RCIM_Cubic, leaveTangent=-0.1241, tangentMode=RCTM_User, time=1, value=1}
         ]
        }}}
      Scale Curve=&Multiply_Float(A=1, B=&RandomRangeFloat(Minimum=3, Maximum=8))
     ))
   ) {isSetParameters=true, module=SetVariables_#12}
   /SetVariables(Particles.CentralFlareColor:LinearColor=&MultiplyLinearColor(
     Linear Color=&ColorFromCurve(
      ColorCurve=DataInterface() {propertyValues {
        AlphaCurve {keys [<at 0 0.7>,<at 0.6638 0.0872>,<at 1 0>]}
        BlueCurve {keys [<at 0 0.1117>,<at 0.4605 0.2551>,<at 1 0>]}
        GreenCurve {keys [<at 0 0.6033>,<at 0.4605 0.5816>,<at 1 0.2045>]}
        RedCurve {keys [<at 0 1>,<at 0.4605 1>,<at 1 1>]}
       }}
      CurveIndex=&Multiply_Float(A=$Particles.NormalizedAge, B=2)
     )
     Scale Factor=&MakeFloat4FromVectorAndFloat("Vector (RGB)"=&VectorFromFloat(Value=2), "Float (Alpha)"=0.5)
    )) {isSetParameters=true}
   /SetVariables(Particles.CentralFlareSize:Vector2f=&Multiply_Vector2D_ByFloat(
     Float=1.5
     Vector2D=&Multiply_Vector2D_ByFloat(
      Float=&FloatFromCurve(
       FloatCurve=DataInterface() {propertyValues {Curve {keys [
           {arriveTangent=0.7585, interpMode=RCIM_Cubic, leaveTangent=0.7585, tangentMode=RCTM_User, time=0, value=0.4}
           {interpMode=RCIM_Cubic, tangentMode=RCTM_User, time=1, value=1}
          ]}}}
       Scale Curve=1.25
      )
      Vector2D=$Particles.Initial.SpriteSize
     )
    )) {isSetParameters=true}
   /SetVariables(
    Particles.Update Impact Linear Color:LinearColor=&MultiplyLinearColor(
     Linear Color=$Particles.Initial.Impact Color
     Scale Factor=&MakeVector4(
      X=1
      Y=0.8
      Z=0.8
      W=&Multiply_Float(
       A=&FloatFromCurve(
        FloatCurve=DataInterface() {propertyValues {Curve {keys [<at 0 0>,<at 0.1 1>,<at 0.2 0.8>,<at 1 0>]}}}
        Scale Curve=&FNTimeOfDayScalar(Night Scale=1, Morning Scale=0.3, Evening Scale=0.5)
       )
       B=0.1
      )
     )
    )
    Particles.Update Impact Size:Vector2f=&Multiply_Vector2D(
     A=$Particles.Initial.ImpactSize
     B=&Vector2DFromCurve(
      Vector2Curve=DataInterface() {propertyValues {
        XCurve {keys [{tangentMode=RCTM_User, time=0, value=1}, {tangentMode=RCTM_User, time=1, value=0.8}]}
        YCurve {keys [{tangentMode=RCTM_User, time=0, value=0.5}, {tangentMode=RCTM_User, time=1, value=1.7}]}
       }}
      ScaleCurve=&Multiply_Vector2D_ByFloat(Float=0.8, Vector2D=<1 1>)
     )
    )
   ) {isSetParameters=true}
  ]
  rendererClasses [&NiagaraSpriteRendererProperties]
  renderers [
   {rendererClass=&NiagaraSpriteRendererProperties}
   {rendererClass=&NiagaraSpriteRendererProperties, rendererIndex=1}
   {rendererClass=&NiagaraSpriteRendererProperties, rendererIndex=2}
   {rendererClass=&NiagaraSpriteRendererProperties, rendererIndex=3}
  ]
 }
 Impact_Spikes {
  bEnabled=false
  emitterUpdateScript [
   &EmitterState(Life Cycle Mode=::NewEnumerator1, Loop Behavior=::NewEnumerator1, Loop Duration=1, UseLoopDelay=false)
   &SpawnRate(
    SpawnGroup=1
    SpawnRate=&FloatFromCurve(
     FloatCurve=DataInterface() {propertyValues {Curve {
        defaultValue=1
        keys [
         {arriveTangent=10.7524, interpMode=RCIM_Cubic, leaveTangent=10.7524, tangentMode=RCTM_User, time=0, value=0}
         <at 0.2657 1.0002 RCIM_Cubic>
         {arriveTangent=-10.45, interpMode=RCIM_Cubic, leaveTangent=-10.45, tangentMode=RCTM_User, time=0.5769, value=0}
        ]
       }}}
     CurveIndex=$Emitter.NormalizedLoopAge
     Scale Curve=15
    )
    Use Spawn Probability=false
   )
  ]
  particleSpawnScript [
   &InitializeParticle(
    Lifetime=&RandomRangeFloat(Minimum=0.2, Maximum=0.15)
    Position Offset=$System.BurstPosition
    Sprite Size=&Multiply_Vector2D_ByFloat(Float=0.6, Vector2D=&RandomRangeVector2D(Minimum=<20 30>, Maximum=<10 20>))
    Sprite Size Mode=::NewEnumerator4
    UsePositionOffset=true
    Write Material Random=true
   )
   &AddVelocityFromPoint(
    Module.Use Falloff Distance=false
    Origin Offset=$System.BurstPosition
    Origin Offset Coordinate Space=::Simulation
    Velocity Strength=&RandomRangeFloat(Minimum=200, Maximum=200)
   )
   &SubUVAnimation(End Frame=3, SubUV Animation Mode:ENiagara_SubUVLookupMode=::NewEnumerator2)
  ]
  particleUpdateScript [
   &ParticleState()
   &Drag(Drag=8, Use Linear Drag=true, Use Rotational Drag=false)
   &SolveForcesAndVelocity(Clamp Velocity=false, Limit Acceleration=false)
   &ScaleSpriteSize(
    Non-Uniform Curve Scale=&Multiply_Vector2D_ByFloat(Float=0.8, Vector2D=<1 1>)
    Non-Uniform Curve Sprite Scale=DataInterface() {propertyValues {
      XCurve {keys [{tangentMode=RCTM_User, time=0, value=1}, {tangentMode=RCTM_User, time=1, value=0.8}]}
      YCurve {keys [{tangentMode=RCTM_User, time=0, value=0.5}, {tangentMode=RCTM_User, time=1, value=1.7}]}
     }}
    Scale Sprite Size Mode=::NewEnumerator3
   )
   &ScaleColor(
    Scale Alpha=&Multiply_Float(
     A=&FloatFromCurve(
      FloatCurve=DataInterface() {propertyValues {Curve {keys [<at 0 0>,<at 0.1 1>,<at 0.2 0.8>,<at 1 0>]}}}
      Scale Curve=&FNTimeOfDayScalar(Night Scale=1, Morning Scale=0.3, Evening Scale=0.5)
     )
     B=0.1
    )
    Scale RGB=<1 0.8 0.8>
    ScaleA=true
    ScaleRGB=true
   )
  ]
  rendererClasses [&NiagaraSpriteRendererProperties]
  renderers [{rendererClass=&NiagaraSpriteRendererProperties}]
 }
 Intro {
  emitterSpawnScript [
   &InitializeAudioPlayer(Audio Player=DataInterface() {propertyValues {
      Attenuation=null
      Concurrency=null
      MaxPlaysPerTick=10
      SoundToPlay=/CosmeticCompanions/Assets/Audio_Shared/Interaction/Petting/SC_Mimosa_Interaction_React_Default
      bAllowLoopingOneShotSounds=false
      bLimitPlaysPerTick=true
      bOnlyActiveDuringGameplay=false
      bStopWhenComponentIsDestroyed=true
      ConfigurationUserParameter {parameter {name=None, typeDefHandle {registeredTypeIndex=1}, varData []}}
      ParameterNames []
     }})
   /Niagara/Modules/Audio/PlayAudio() {module=PlayAudio001}
  ]
  emitterUpdateScript [
   &EmitterState()
   &SpawnBurst_Instantaneous(Use Spawn Probability=false, UseLoopCountLimit=false)
  ]
  particleSpawnScript [&InitializeParticle(
    Lifetime=$System.IntroLength
    Sprite Rotation Mode=::NewEnumerator0
    Sprite Size Mode=::NewEnumerator3
    Uniform Sprite Size=15
    UsePositionOffset=false
    Write Material Random=true
   )]
  particleUpdateScript [
   &ParticleState()
   &ScaleSpriteSize(
    Scale Sprite Size Mode=::NewEnumerator0
    Uniform Scale Factor=&Add_Float(
     A=0.5
     B=&FloatFromCurve(FloatCurve=DataInterface() {propertyValues {Curve {
         defaultValue=0
         keys [
          {arriveTangent=0.8193, interpMode=RCIM_Cubic, leaveTangent=0.8193, tangentMode=RCTM_User, time=0, value=0}
          {arriveTangent=2.055, interpMode=RCIM_Cubic, leaveTangent=2.055, tangentMode=RCTM_User, time=0.418, value=1.8546}
          {arriveTangent=-0.9053, interpMode=RCIM_Cubic, leaveTangent=-0.9053, tangentMode=RCTM_User, time=0.8783, value=2}
          {arriveTangent=-0.1202, interpMode=RCIM_Cubic, leaveTangent=-0.1202, tangentMode=RCTM_User, time=1, value=1}
         ]
        }}})
    )
   ) {module=ScaleSpriteSize002}
   &ScaleColor(
    Scale Alpha=&FloatFromCurve(FloatCurve=DataInterface() {propertyValues {Curve {
        defaultValue=0
        keys [
         {arriveTangent=9.0725, interpMode=RCIM_Cubic, leaveTangent=9.0725, tangentMode=RCTM_User, time=0, value=0}
         {arriveTangent=-0.2441, leaveTangent=-0.2441, time=0.1408, value=0.9847}
         {arriveTangent=-1.146, leaveTangent=-1.146, time=0.95, value=1}
         {arriveTangent=-0.0405, interpMode=RCIM_Cubic, leaveTangent=-0.0405, tangentMode=RCTM_User, time=1, value=0}
        ]
       }}})
    ScaleA=true
    ScaleRGB=false
   )
   &OffsetPosition(Position Offset=$System.BurstPosition)
   &JitterPosition(Jitter Amount=2)
   /SetVariables(Particles.Position=&LerpPosition(
     A=Linked() {linkedVariable {name=System.BurstPosition, type=Vector3f}}
     B=$StackContext.Position
     Alpha=&FloatFromCurve(FloatCurve=DataInterface() {propertyValues {Curve {
         defaultValue=1
         keys [
          {arriveTangent=0.3594, interpMode=RCIM_Cubic, leaveTangent=0.3594, tangentMode=RCTM_User, time=0, value=0}
          {arriveTangent=1.5647, interpMode=RCIM_Cubic, leaveTangent=1.5647, time=0.3969, value=0.5775}
          {arriveTangent=-0.024, interpMode=RCIM_Cubic, leaveTangent=-0.024, tangentMode=RCTM_User, time=0.7056, value=1.1041}
          {arriveTangent=0.4707, interpMode=RCIM_Cubic, leaveTangent=0.4707, tangentMode=RCTM_User, time=0.8025, value=0}
         ]
        }}})
    )) {isSetParameters=true}
   &DynamicMaterialParameters(
    Index 0 Param 1=&FloatFromCurve(FloatCurve=DataInterface() {propertyValues {Curve {keys [{arriveTangent=0.6296, interpMode=RCIM_Cubic, leaveTangent=0.6296, tangentMode=RCTM_User, time=0, value=0}, <at 0.75 0.9676>, <at 1 1>]}}})
    Index 0 Param 2=1
    Module.Index0Float0Write=true
    Module.Index0Float1Write=true
    Module.Index0Float2Write=true
    Module.Index0Float3Write=true
   )
   /SetVariables(
    Particles.Glow Color=&Multiply_LinearColorByFloat(Linear Color=$System.Layer 2 Color, Float=0.3)
    Particles.Glow Size=&Multiply_Vector2D(A=$Particles.SpriteSize, B=<0.6 0.6>)
   ) {isSetParameters=true}
  ]
  rendererClasses [&NiagaraSpriteRendererProperties]
  renderers [
   {rendererClass=&NiagaraSpriteRendererProperties}
   {rendererClass=&NiagaraSpriteRendererProperties, rendererIndex=1}
  ]
 }
 NE_First_Flare {
  bEnabled=false
  emitterUpdateScript [
   &EmitterState(Life Cycle Mode=::NewEnumerator1, Loop Behavior=::NewEnumerator1, Loop Delay=$System.IntroLength, Loop Duration=0.75, UseLoopDelay=true)
   &SpawnBurst_Instantaneous(Use Spawn Probability=false, UseLoopCountLimit=false)
  ]
  particleSpawnScript [
   &InitializeParticle(
    Color=<rgba 0.35 0.35 0.35 1>
    Lifetime=0.4
    Mass Mode=::NewEnumerator1
    Mesh Scale Mode=::NewEnumerator3
    Mesh Uniform Scale=&FloatFromCurve(
     FloatCurve=DataInterface() {propertyValues {Curve {keys [{tangentMode=RCTM_User, time=0, value=1}, {tangentMode=RCTM_User, time=1, value=2}]}}}
     CurveIndex=&ReturnNormalizedExecIndex()
     Scale Curve=1.5
    )
    Position Offset:NiagaraPosition=$System.BurstPosition
    Sprite Size=<100 100>
    Sprite Size Mode=::NewEnumerator4
    UsePositionOffset=true
    Write Material Random=true
   ) {module=InitializeParticle001}
   &CameraOffset(Camera Offset Amount=5)
  ]
  particleUpdateScript [
   &ParticleState()
   &SolveForcesAndVelocity(Clamp Velocity=false, Limit Acceleration=false)
   &ScaleColor(
    Scale Alpha=&FloatFromCurve(
     FloatCurve=DataInterface() {propertyValues {Curve {keys [
         {arriveTangent=6.4019, interpMode=RCIM_Cubic, leaveTangent=6.4019, tangentMode=RCTM_User, time=0, value=0}
         {arriveTangent=0.002, interpMode=RCIM_Cubic, leaveTangent=0.002, time=0.2, value=0.7}
         {arriveTangent=-0.9006, interpMode=RCIM_Cubic, leaveTangent=-0.9006, tangentMode=RCTM_User, time=1.0011, value=0.002}
        ]}}}
     Scale Curve=&FNTimeOfDayScalar(Night Scale=0.5, Morning Scale=0.7, Evening Scale=0.7)
    )
    Scale RGB=&MakeVectorFromLinearColorRGB(Color=&FNTimeOfDayColor(Night Color=<rgba 0.5006 0.8335 1 1>, Morning Color=<rgba 1 1 1 1>, Day Color=<rgba 1 1 1 1>, Evening Color=<rgba 1 1 1 1>))
    ScaleA=true
    ScaleRGB=true
   )
   &ScaleSpriteSize(Scale Factor=&Multiply_Vector2D_ByFloat(
     Float=&FloatFromCurve(
      FloatCurve=DataInterface() {propertyValues {Curve {keys [
          {arriveTangent=0.7585, interpMode=RCIM_Cubic, leaveTangent=0.7585, tangentMode=RCTM_User, time=0, value=0.4}
          {interpMode=RCIM_Cubic, tangentMode=RCTM_User, time=1, value=1}
         ]}}}
      Scale Curve=1.25
     )
     Vector2D=<1 1>
    ))
   /SetVariables(Particles.CentralFlareSize:Vector2f=&Multiply_Vector2D_ByFloat(Float=1.5, Vector2D=$Particles.SpriteSize)) {isSetParameters=true}
   /SetVariables(Particles.CentralFlareColor:LinearColor=&MultiplyLinearColor(
     Linear Color=&ColorFromCurve(
      ColorCurve=DataInterface() {propertyValues {
        AlphaCurve {keys [<at 0 0.7>,<at 0.6638 0.0872>,<at 1 0>]}
        BlueCurve {keys [<at 0 0.1117>,<at 0.4605 0.2551>,<at 1 0>]}
        GreenCurve {keys [<at 0 0.6033>,<at 0.4605 0.5816>,<at 1 0.2045>]}
        RedCurve {keys [<at 0 1>,<at 0.4605 1>,<at 1 1>]}
       }}
      CurveIndex=&Multiply_Float(A=$Particles.NormalizedAge, B=2)
     )
     Scale Factor=&MakeFloat4FromVectorAndFloat("Vector (RGB)"=&VectorFromFloat(Value=2), "Float (Alpha)"=0.5)
    )) {isSetParameters=true}
  ]
  rendererClasses [&NiagaraSpriteRendererProperties]
  renderers [{rendererClass=&NiagaraSpriteRendererProperties}]
 }
}
summary {
 systemName=NS_ReactFX_TeaPup
 emitters [
  {emitterName=HeartsFlare, rendererClasses [&NiagaraSpriteRendererProperties]}
  {emitterName=Intro, rendererClasses [&NiagaraSpriteRendererProperties]}
  {bEnabled=false, emitterName=NE_First_Flare, rendererClasses [&NiagaraSpriteRendererProperties]}
  {bEnabled=false, emitterName=Impact_Spikes, rendererClasses [&NiagaraSpriteRendererProperties]}
  {bEnabled=false, emitterName=Hearts001, rendererClasses [&NiagaraSpriteRendererProperties]}
 ]
 userVariables [
  User.bIsFrontEnd {type=NiagaraBool}
  User.bIsFrontEndPreview {type=NiagaraBool}
 ]
}
