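# Niagara system for the ViperBucket cosmetic companion's reaction effect, pairing a heart intro with a snow dust puff.
# Intro_Heart001 plays the petting reaction sound and bursts a curve-scaled, color-animated heart at System.BurstPosition.
# snow_dust_drop spawns falling sprite and mesh flecks in a brief age-gated window near that position.
# Front-end user flags and modifier inputs are present; six LW and SnowMesh emitters list no renderers.
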NiagaraSystem /CosmeticCompanions/FX/React/NS_ReactFX_ViperBucket
dependencies {
 usedDataInterfaces [
  &NiagaraDataInterfaceAudioPlayer
  &NiagaraDataInterfaceCurve
  &NiagaraDataInterfaceVector2DCurve
  &NiagaraDataInterfaceCamera
 ]
 usedDynamicInputs [
  &Multiply_Vector2D_ByFloat
  /Niagara/DynamicInputs/Add/Add_Vector2
  &Vector2DFromCurve
  &RandomRangeVector
  &ReturnCameraVector
  &FN_FrontEndVector3Modifier
  &FN_FrontEndFloatModifier
 ]
 usedModules [
  &self:SystemSpawnScript.SystemScriptSource.SystemScriptGraph.NiagaraNodeAssignment_0.SetVariables_#1
  &self:SystemSpawnScript.SystemScriptSource.SystemScriptGraph.NiagaraNodeAssignment_2.SetVariables_#2
  &self:SystemSpawnScript.SystemScriptSource.SystemScriptGraph.NiagaraNodeAssignment_4.SetVariables_#3
  &self:SystemSpawnScript.SystemScriptSource.SystemScriptGraph.NiagaraNodeAssignment_7.SetVariables_#4
  &self:SystemSpawnScript.SystemScriptSource.SystemScriptGraph.NiagaraNodeAssignment_1.SetVariables_#5
  &SystemState
 ]
}
emitters {
 Intro_Heart001 {
  emitterSpawnScript [
   &InitializeAudioPlayer(Audio Player=DataInterface() {propertyValues {
      Attenuation=null
      Concurrency=null
      MaxPlaysPerTick=10
      SoundToPlay=/CosmeticCompanions/Assets/Audio_Shared/Interaction/Petting/SC_Mimosa_Interaction_React_Default
      bAllowLoopingOneShotSounds=false
      bLimitPlaysPerTick=true
      bOnlyActiveDuringGameplay=false
      bStopWhenComponentIsDestroyed=true
      ConfigurationUserParameter {parameter {name=None, typeDefHandle {registeredTypeIndex=1}, varData []}}
      ParameterNames []
     }})
   /Niagara/Modules/Audio/PlayAudio() {module=PlayAudio001}
  ]
  emitterUpdateScript [
   &EmitterState(Life Cycle Mode=::NewEnumerator1, Loop Behavior=::NewEnumerator1, Loop Duration=1.25, UseLoopDelay=false)
   &SpawnBurst_Instantaneous(Use Spawn Probability=false, UseLoopCountLimit=false)
  ]
  particleSpawnScript [&InitializeParticle(
    Color Mode=::NewEnumerator0
    Lifetime=$System.IntroLength
    Position Offset=$System.BurstPosition
    Sprite Rotation Mode=::NewEnumerator0
    Sprite Size Mode=::NewEnumerator3
    Uniform Sprite Size=15
    UsePositionOffset=true
    Write Material Random=true
   )]
  particleUpdateScript [
   &ParticleState()
   &ScaleSpriteSize(
    Scale Sprite Size Mode=::NewEnumerator0
    Uniform Scale Factor=&FloatFromCurve(
     FloatCurve=DataInterface() {propertyValues {Curve {keys [
         {arriveTangent=1.2333, interpMode=RCIM_Cubic, leaveTangent=1.2333, tangentMode=RCTM_User, time=0, value=0.2}
         {arriveTangent=5.6726, interpMode=RCIM_Cubic, leaveTangent=5.6726, time=0.0414, value=0.2829}
         {arriveTangent=4.8098, interpMode=RCIM_Cubic, leaveTangent=4.8098, time=0.0947, value=0.7374}
         {arriveTangent=1.1785, interpMode=RCIM_Cubic, leaveTangent=1.1785, tangentMode=RCTM_User, time=0.1905, value=1}
         {arriveTangent=0.1199, interpMode=RCIM_Constant, tangentMode=RCTM_User, time=0.8075, value=1.4287}
        ]}}}
     CurveIndex=$Particles.Age
    )
   )
   &OffsetPosition(Position Offset=&Lerp_Vector(
     A=$System.HeartIntroSpawnPosition
     B=$System.BurstPosition
     Alpha=&FloatFromCurve(FloatCurve=DataInterface() {propertyValues {Curve {defaultValue=0, keys [{arriveTangent=2.9288, interpMode=RCIM_Cubic, leaveTangent=2.9288, tangentMode=RCTM_User, time=0, value=0}, <at 1 1>]}}})
    )) {enabled=false}
   /SetVariables(Particles.HeartWhite_DMP:Vector4f=&MakeVector4(
     X=&FloatFromCurve(
      FloatCurve=DataInterface() {propertyValues {Curve {keys [{arriveTangent=0.1332, time=0.3, value=1.2}, <at 0.8 0>]}}}
      CurveIndex=$Particles.Age
     )
     Y=0.3
     Z=0
     W=0
    )) {isSetParameters=true}
   /SetVariables(
    Particles.Color_Heart:LinearColor=&LerpLinearColors(
     StartColor=<rgba 0.5 0 0 1>
     EndColor=<rgba 0.5029 0 0 1>
     LerpFactor=&FloatFromCurve(FloatCurve=DataInterface() {propertyValues {Curve {keys [{arriveTangent=3.3745, time=0.75, value=0}, {arriveTangent=-0.0365, leaveTangent=-0.0365, time=0.9, value=1}]}}})
    )
    Particles.Color_HeartGlow:LinearColor=&LerpLinearColors(
     StartColor=<rgba 1 0 0 0.35>
     EndColor=<rgba 1 0 0 0>
     LerpFactor=&FloatFromCurve(FloatCurve=DataInterface() {propertyValues {Curve {keys [{arriveTangent=3.3745, time=0, value=0}, {arriveTangent=-0.0365, leaveTangent=-0.0365, time=0.5, value=1}]}}})
    )
   ) {isSetParameters=true}
  ]
  rendererClasses [&NiagaraSpriteRendererProperties]
  renderers [
   {rendererClass=&NiagaraSpriteRendererProperties}
   {rendererClass=&NiagaraSpriteRendererProperties, rendererIndex=1}
  ]
 }
 LW_BlizzardErode {rendererClasses [], renderers []}
 LW_HaloBurst {rendererClasses [], renderers []}
 LW_SnowFlakes {rendererClasses [], renderers []}
 LW_WispSmokeImpact {rendererClasses [], renderers []}
 LW_Wolf_Icon {rendererClasses [], renderers []}
 SnowMesh_SnowSpriteCluster {rendererClasses [], renderers []}
 snow_dust_drop {
  emitterUpdateScript [
   &EmitterState(Life Cycle Mode=::NewEnumerator1, Loop Behavior=::NewEnumerator1, Loop Duration=2.75, UseLoopDelay=false)
   &SpawnRate(
    SpawnRate=&MakeCustomFloatFromBool(
     Bool=&Boolean_AndOperation(
      Bool1=&SetBoolByFloatComparison(A=$Emitter.Age, B=1.6)
      Bool2=&SetBoolByFloatComparison(A=$Emitter.Age, B=1.7, Comparison Type=::NewEnumerator6)
     )
     True Float=70
     False Float=0
    )
    Use Spawn Probability=false
   )
  ]
  particleSpawnScript [
   &InitializeParticle(
    Color=&RandomRangeLinearColor(Minimum=<rgba 0.4295 0.8406 0.8362 1>, Maximum=<rgba 0.9073 0.9073 0.9073 1>)
    Lifetime=&RandomRangeFloat(Minimum=0.4, Maximum=0.8)
    Mesh Scale Mode=::NewEnumerator1
    Mesh Uniform Scale Max=0.02
    Mesh Uniform Scale Min=0.005
    Position Offset=&AddVectorToPosition(Position=Linked() {linkedVariable {name=System.BurstPosition, type=Vector3f}}, Vector=<0 0 7>)
    Sprite Rotation Mode=::NewEnumerator2
    Sprite Size Mode=::NewEnumerator3
    Sprite UV Mode=::NewEnumerator3
    Uniform Sprite Size=&RandomRangeFloat(Minimum=20, Maximum=30)
    UsePositionOffset=true
    Write Material Random=true
   )
   &ShapeLocation(
    Hemisphere Distribution=<1 0.2>
    MaskSpawnGroup=false
    Non Uniform Scale=<1 1 0.22>
    Override Seed=false
    Sphere Radius=23
    Sphere Surface Distribution=1
    UseOwnerScale=true
   )
  ]
  particleUpdateScript [
   &ParticleState()
   &ScaleSpriteSize(Uniform Curve Sprite Scale=DataInterface() {propertyValues {Curve {
       defaultValue=0
       keys [
        {arriveTangent=2.4405, interpMode=RCIM_Cubic, leaveTangent=2.4405, tangentMode=RCTM_User, time=0, value=0}
        <at 0.4 1 RCIM_Cubic>
        {arriveTangent=-3.1685, tangentMode=RCTM_User, time=1, value=0}
       ]
      }}})
   &GravityForce(Gravity=<0 0 -300>)
   &Drag(Drag=&Lerp_Float(A=3, B=0, Alpha=$Particles.NormalizedAge), Use Linear Drag=true, Use Rotational Drag=false)
   &SolveForcesAndVelocity(Clamp Velocity=false, Limit Acceleration=false)
   &ScaleColor(
    Scale Alpha=&FloatFromCurve(
     FloatCurve=DataInterface() {propertyValues {Curve {keys [<at 0 0>, {arriveTangent=-1.6625, leaveTangent=-1.6625, time=0.15, value=1}, <at 0.5 1>, {arriveTangent=-0.5995, time=1, value=0}]}}}
     Scale Curve=0.9
    )
    ScaleA=true
    ScaleRGB=false
   )
   &ScaleMeshSize(Scale Factor=&Multiply_VectorByFloat(
     Vector=<1 1 1>
     Float=&FloatFromCurve(FloatCurve=DataInterface() {propertyValues {Curve {keys [
          {arriveTangent=0.6971, time=0, value=0}
          {arriveTangent=-0.0529, leaveTangent=-0.0529, time=0.2, value=1}
          {arriveTangent=-1.6278, leaveTangent=-1.6278, time=0.5, value=1}
          {arriveTangent=-0.0405, leaveTangent=-0.0405, time=1, value=0}
         ]}}})
    ))
  ]
  rendererClasses [&NiagaraSpriteRendererProperties, &NiagaraMeshRendererProperties]
  renderers [
   {rendererClass=&NiagaraSpriteRendererProperties}
   {rendererClass=&NiagaraMeshRendererProperties, rendererIndex=1}
  ]
 }
}
summary {
 systemName=NS_ReactFX_ViperBucket
 emitters [
  {emitterName=Intro_Heart001, rendererClasses [&NiagaraSpriteRendererProperties]}
  {emitterName=snow_dust_drop, rendererClasses [&NiagaraSpriteRendererProperties, &NiagaraMeshRendererProperties]}
  {emitterName=LW_HaloBurst, rendererClasses []}
  {emitterName=LW_Wolf_Icon, rendererClasses []}
  {emitterName=LW_SnowFlakes, rendererClasses []}
  {emitterName=LW_BlizzardErode, rendererClasses []}
  {emitterName=LW_WispSmokeImpact, rendererClasses []}
  {emitterName=SnowMesh_SnowSpriteCluster, rendererClasses []}
 ]
 userVariables [
  User.bIsFrontEnd {type=NiagaraBool}
  User.bIsFrontEndPreview {type=NiagaraBool}
 ]
}
