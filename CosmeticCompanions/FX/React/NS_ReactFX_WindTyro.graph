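# Niagara reaction effect for the WindTyro cosmetic companion, showing a heart and three cogs with a shared petting-reaction sound.
# The heart slides along a camera axis to the burst position while spinning cogs follow randomized Hermite splines to camera-relative offsets around it.
# Ramping sparks, cross stars, a glow pop and a light build the climax.
# Front-end modifiers rescale sizes and motion, and time-of-day inputs adjust cog color and heart material.
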NiagaraSystem /CosmeticCompanions/FX/React/NS_ReactFX_WindTyro
dependencies {
 usedDataInterfaces [
  &NiagaraDataInterfaceAudioPlayer
  &NiagaraDataInterfaceCurve
  &NiagaraDataInterfaceCamera
  &NiagaraDataInterfaceArrayInt32
  &NiagaraDataInterfaceColorCurve
  &NiagaraDataInterfaceVectorCurve
 ]
 usedDynamicInputs [
  &ReturnNormalizedExecIndex
  &FN_FrontEndVector3Modifier
  &ColorFromCurve
 ]
 usedModules [
  &self:SystemSpawnScript.SystemScriptSource.SystemScriptGraph.NiagaraNodeAssignment_0.SetVariables_#1
  &self:SystemSpawnScript.SystemScriptSource.SystemScriptGraph.NiagaraNodeAssignment_1.SetVariables_#2
  &self:SystemSpawnScript.SystemScriptSource.SystemScriptGraph.NiagaraNodeAssignment_4.SetVariables_#3
  &self:SystemSpawnScript.SystemScriptSource.SystemScriptGraph.NiagaraNodeAssignment_5.SetVariables_#4
  &self:SystemSpawnScript.SystemScriptSource.SystemScriptGraph.NiagaraNodeAssignment_6.SetVariables_#5
  &self:SystemSpawnScript.SystemScriptSource.SystemScriptGraph.NiagaraNodeAssignment_2.SetVariables_#6
  &self:SystemSpawnScript.SystemScriptSource.SystemScriptGraph.NiagaraNodeAssignment_3.SetVariables_#7
  &SystemState
 ]
}
emitters {
 Cogs {
  emitterUpdateScript [
   &EmitterState()
   &SpawnBurst_Instantaneous(Spawn Count=3, Use Spawn Probability=false, UseLoopCountLimit=false)
  ]
  particleSpawnScript [
   &PartitionParticles(Number Of Partitions=3)
   &InitializeParticle(
    Color=<rgba 2 0.575 0.1 1>
    Lifetime=$System.IntroLength
    Mesh Scale Mode=::NewEnumerator3
    Mesh Uniform Scale=&MergeHelper_FloatSwitch(0=1.2, 1=1.5, 2=0.9, 3=?, Index=$Particles.PartitionParticles.Partition, Num=::NewEnumerator1)
    Sprite Size Mode=::NewEnumerator3
    Uniform Sprite Size=&Multiply_Float(
     A=10
     B=&MergeHelper_FloatSwitch(0=2, 1=2.5, 2=1.5, 3=?, Index=$Particles.PartitionParticles.Partition, Num=::NewEnumerator1)
    )
    UsePositionOffset=false
    Write Material Random=true
   )
   &InitialMeshOrientation(
    Mesh Axis To Orient=&Add_Vector(A=<0 0 1>, B=&RandomVector(Vector Scale=0))
    Mesh Orientation Mode=::NewEnumerator2
    Use Rotation Vector=true
   )
   &CameraOffset(Camera Offset Amount=5)
  ]
  particleUpdateScript [
   &ParticleState()
   &ScaleMeshSize(Scale Factor=&VectorFromFloat(Value=&FloatFromCurve(
      FloatCurve=DataInterface() {propertyValues {Curve {
         defaultValue=1
         keys [
          {arriveTangent=4.1901, interpMode=RCIM_Cubic, leaveTangent=4.1901, tangentMode=RCTM_User, time=0, value=0}
          {interpMode=RCIM_Cubic, tangentMode=RCTM_User, time=0.4, value=1}
          {arriveTangent=-0.0059, interpMode=RCIM_Cubic, leaveTangent=-0.0059, tangentMode=RCTM_User, time=0.975, value=1}
          <at 1 0 RCIM_Cubic>
         ]
        }}}
      Scale Curve=&FN_FrontEndFloatModifier(Gameplay=1, Front End=$System.FrontendScale)
     )))
   &ScaleSpriteSize(
    Scale Sprite Size Mode=::NewEnumerator0
    Uniform Scale Factor=&FloatFromCurve(
     FloatCurve=DataInterface() {propertyValues {Curve {
        defaultValue=1
        keys [
         {arriveTangent=4.1901, interpMode=RCIM_Cubic, leaveTangent=4.1901, tangentMode=RCTM_User, time=0, value=0}
         {interpMode=RCIM_Cubic, tangentMode=RCTM_User, time=0.4, value=1}
         {arriveTangent=-0.0059, interpMode=RCIM_Cubic, leaveTangent=-0.0059, tangentMode=RCTM_User, time=0.9, value=1}
         <at 1 0 RCIM_Cubic>
        ]
       }}}
     Scale Curve=&FN_FrontEndFloatModifier(Gameplay=1, Front End=0.6)
    )
   )
   /Game/Effects/Niagara/Modules/Curves/FN_HermiteSpline(
    Direction End:Vector3f=&MakeVector(
     X=&Multiply_FloatByInt(
      Float=&RandomRangeFloat(Minimum=50, Maximum=60)
      Integer=&SelectIntFromArray(
       Array Sampling Mode=::NewEnumerator1
       Direct Array Index=?
       Int Selection Array=DataInterface() {propertyValues {IntData [-1,1]}}
       Random Seed=0
      )
     )
     Y=&Multiply_FloatByInt(
      Float=&RandomRangeFloat(Minimum=50, Maximum=60)
      Integer=&SelectIntFromArray(
       Array Sampling Mode=::NewEnumerator1
       Direct Array Index=?
       Int Selection Array=DataInterface() {propertyValues {IntData [-1,1]}}
       Random Seed=0
      )
     )
     Z=-70
    )
    Direction Start:Vector3f=&MakeVector(
     X=&Multiply_FloatByInt(
      Float=&RandomRangeFloat(Minimum=50, Maximum=60)
      Integer=&SelectIntFromArray(
       Array Sampling Mode=::NewEnumerator1
       Direct Array Index=?
       Int Selection Array=DataInterface() {propertyValues {IntData [-1,1]}}
       Random Seed=0
      )
     )
     Y=&Multiply_FloatByInt(
      Float=&RandomRangeFloat(Minimum=50, Maximum=60)
      Integer=&SelectIntFromArray(
       Array Sampling Mode=::NewEnumerator1
       Direct Array Index=?
       Int Selection Array=DataInterface() {propertyValues {IntData [-1,1]}}
       Random Seed=0
      )
     )
     Z=50
    )
    Position Blend:NiagaraFloat=&FloatFromCurve(FloatCurve=DataInterface() {propertyValues {Curve {
        defaultValue=0.35
        keys [{arriveTangent=2.4518, interpMode=RCIM_Cubic, leaveTangent=2.4518, tangentMode=RCTM_User, time=0, value=0.3}, <at 0.8 1 RCIM_Cubic>]
       }}})
    Position End:Vector3f=&Add_Vector(
     A=$System.BurstPosition
     B=&Add_Vector(
      A=&Multiply_VectorByFloat(
       Vector=&ReturnCameraVector(Camera Vectors=::NewEnumerator3, Coordinate Space=::Simulation)
       Float=&Multiply_Float(
        A=&FN_FrontEndFloatModifier(Gameplay=1, Front End=$System.FrontendScale)
        B=&MergeHelper_FloatSwitch(0=5.4, 1=-4.8, 2=0.6, 3=?, Index=$Particles.PartitionParticles.Partition, Num=::NewEnumerator1)
       )
      )
      B=&Multiply_VectorByFloat(
       Vector=&ReturnCameraVector(Camera Vectors=::NewEnumerator2, Coordinate Space=::Simulation)
       Float=&Multiply_Float(
        A=&FN_FrontEndFloatModifier(Gameplay=1, Front End=$System.FrontendScale)
        B=&MergeHelper_FloatSwitch(0=1.8, 1=3, 2=-2.4, 3=?, Index=$Particles.PartitionParticles.Partition, Num=::NewEnumerator1)
       )
      )
     )
    )
    Position Start:Vector3f=$System.HeartIntroSpawnPosition
    Write Particles Position=true
   )
   &ScaleColor(
    Scale Alpha=&FloatFromCurve(
     FloatCurve=DataInterface() {propertyValues {Curve {
        defaultValue=0
        keys [
         {arriveTangent=2.1412, interpMode=RCIM_Cubic, leaveTangent=2.1412, tangentMode=RCTM_User, time=0, value=0}
         {interpMode=RCIM_Cubic, tangentMode=RCTM_User, time=0.9, value=1}
         <at 1 0 RCIM_Cubic>
        ]
       }}}
     Scale Curve=&FN_FrontEndFloatModifier(Gameplay=1, Front End=0.5)
    )
    ScaleA=true
    ScaleRGB=true
   )
   &DynamicMaterialParameters(
    Index 0 Param 1=&FloatFromCurve(FloatCurve=DataInterface() {propertyValues {Curve {keys [{arriveTangent=0.6296, interpMode=RCIM_Cubic, leaveTangent=0.6296, tangentMode=RCTM_User, time=0.5, value=1}, <at 1 2>]}}})
    Index 0 Param 2=&FloatFromCurve(FloatCurve=DataInterface() {propertyValues {Curve {keys [{arriveTangent=0.6296, interpMode=RCIM_Cubic, leaveTangent=0.6296, tangentMode=RCTM_User, time=0.5, value=1}, <at 0.6 50>, <at 0.7 1>]}}})
    Index 0 Param 3=1
    Module.Index0Float0Write=true
    Module.Index0Float1Write=true
    Module.Index0Float2Write=true
    Module.Index0Float3Write=true
   )
   &UpdateMeshOrientation(
    Rotation Rate=&Multiply_Float(
     A=&FloatFromCurve(
      FloatCurve=DataInterface() {propertyValues {Curve {
         defaultValue=1
         keys [
          {arriveTangent=0.0104, interpMode=RCIM_Cubic, leaveTangent=0.0104, tangentMode=RCTM_User, time=0, value=0.1}
          {arriveTangent=1.6105, interpMode=RCIM_Cubic, leaveTangent=1.6105, tangentMode=RCTM_User, time=0.4, value=0.4}
          {interpMode=RCIM_Cubic, tangentMode=RCTM_User, time=0.9, value=1}
         ]
        }}}
      Scale Curve=0.07
     )
     B=&MergeHelper_FloatSwitch(0=15, 1=12, 2=-20, 3=?, Index=$Particles.PartitionParticles.Partition, Num=::NewEnumerator1)
    )
    Rotation Vector=<0 0 1>
   )
   &FN_Mesh_Attributes(
    Color=&FNTimeOfDayColor(
     Night Color=<rgba 1.2109 0.1621 0.0883 1>
     Morning Color=<rgba 1.2109 0.1621 0.0883 1>
     Day Color=<rgba 1 0.1549 0.0781 1>
     Evening Color=<rgba 1.2109 0.1621 0.0883 1>
    )
    Create Color Attribute=true
    Write to Module Namespace=true
   )
   &JitterPosition(
    Jitter Amount:NiagaraFloat=&FloatFromCurve(
     FloatCurve=DataInterface() {propertyValues {Curve {
        defaultValue=1
        keys [{interpMode=RCIM_Cubic, tangentMode=RCTM_User, time=0.2, value=0}, {interpMode=RCIM_Cubic, tangentMode=RCTM_User, time=0.8, value=1}]
       }}}
     Scale Curve=1.2
    )
    Jitter Delay=0.01
   )
  ]
  rendererClasses [&NiagaraSpriteRendererProperties, &NiagaraMeshRendererProperties]
  renderers [
   {rendererClass=&NiagaraSpriteRendererProperties}
   {rendererClass=&NiagaraMeshRendererProperties, rendererIndex=1}
  ]
 }
 CrossStars {
  emitterUpdateScript [
   &EmitterState(Life Cycle Mode=::NewEnumerator1, Loop Behavior=::NewEnumerator1, Loop Duration=1, UseLoopDelay=false)
   &SpawnRate(
    SpawnRate=&FloatFromCurve(
     FloatCurve=DataInterface() {propertyValues {Curve {
        defaultValue=1
        keys [
         {interpMode=RCIM_Cubic, tangentMode=RCTM_User, time=0.55, value=0}
         {interpMode=RCIM_Cubic, tangentMode=RCTM_User, time=0.6, value=1}
         <at 0.7 1 RCIM_Constant>
         <at 0.75 0 RCIM_Constant>
        ]
       }}}
     CurveIndex=$Emitter.NormalizedLoopAge
     Scale Curve=20
    )
    Use Spawn Probability=false
   )
   &SpawnBurst_Instantaneous(
    Spawn Count=&RandomRangeInt(Minimum=10, Maximum=14)
    Spawn Group=1
    Spawn Time=0.9
    Use Spawn Probability=false
    UseLoopCountLimit=false
   )
  ]
  particleSpawnScript [
   &InitializeParticle(
    Color=<rgba 1.5 0.5511 0.2348 1>
    Lifetime Max=0.5
    Lifetime Min=0.3
    Lifetime Mode=::NewEnumerator1
    Position Offset=$System.BurstPosition
    Sprite Size Mode=::NewEnumerator1
    UsePositionOffset=true
    Write Material Random=true
   )
   &CameraOffset(Camera Offset Amount=5)
   &ShapeLocation(
    MaskSpawnGroup=false
    Override Seed=false
    Sphere Radius=&Multiply_Float(
     A=&MaskFloatBySpawnGroup(
      Masked Float=&RandomRangeFloat(Minimum=12, Maximum=15)
      Passthrough Float=&RandomRangeFloat(Minimum=1.8, Maximum=3)
     )
     B=&FN_FrontEndFloatModifier(Gameplay=1, Front End=$System.FrontendScale)
    )
    Sphere Surface Distribution:NiagaraFloat=&MaskFloatBySpawnGroup(Masked Float=0.5, Passthrough Float=0)
    UseOwnerScale=true
   )
   &AddVelocity(
    Module.Use Falloff Distance=false
    Origin Offset=$System.BurstPosition
    Override Seed=false
    Velocity Mode=::NewEnumerator1
    Velocity Speed=&Multiply_Float(
     A=&FN_FrontEndFloatModifier(Gameplay=1, Front End=$System.FrontendScale)
     B=&MaskFloatBySpawnGroup(
      Masked Float=&RandomRangeFloat(Minimum=18, Maximum=36)
      Passthrough Float=&RandomRangeFloat(Minimum=60, Maximum=180)
     )
    )
   )
  ]
  particleUpdateScript [
   &ParticleState()
   &ScaleSpriteSize(
    Uniform Curve Scale=0.6
    Uniform Curve Sprite Scale=DataInterface() {propertyValues {Curve {
       defaultValue=1
       keys [
        {arriveTangent=2.5042, interpMode=RCIM_Cubic, leaveTangent=2.5042, tangentMode=RCTM_User, time=0, value=0.6}
        {arriveTangent=-1.6628, tangentMode=RCTM_User, time=1, value=0.2}
       ]
      }}}
   )
   &ScaleColor(
    Scale Alpha=&FloatFromCurve(FloatCurve=DataInterface() {propertyValues {Curve {
        defaultValue=0
        keys [
         {interpMode=RCIM_Cubic, tangentMode=RCTM_User, time=0, value=1}
         {arriveTangent=-2.2441, interpMode=RCIM_Cubic, leaveTangent=-2.2441, tangentMode=RCTM_User, time=1, value=0}
        ]
       }}})
    ScaleA=true
    ScaleRGB=true
   )
   &DynamicMaterialParameters(
    Index 0 Param 1=&Add_Float(
     A=&RandomRangeFloat(Minimum=0, Maximum=1)
     B=&FloatFromCurve(
      FloatCurve=DataInterface() {propertyValues {Curve {keys [
          <at 0 0 RCIM_Cubic>
          {arriveTangent=1.5947, interpMode=RCIM_Cubic, leaveTangent=1.5947, tangentMode=RCTM_User, time=0.2557, value=0.5257}
          <at 1 1 RCIM_Cubic>
         ]}}}
      CurveIndex=$Particles.Age
      Scale Curve=0.25
     )
    )
    Module.Param0WriteEnabled=true
    Module.Param1WriteEnabled=true
    Module.Param2WriteEnabled=true
    Module.Param3WriteEnabled=true
   ) {module=DynamicMaterialParameters001}
   &Drag(
    Drag=&Multiply_Float(
     A=&RandomRangeFloat(Minimum=5, Maximum=7)
     B=&FN_FrontEndFloatModifier(Gameplay=1, Front End=$System.FrontendScale)
    )
    Use Linear Drag=true
    Use Rotational Drag=false
   )
   &CurlNoiseForce(
    Bias Noise=false
    Noise Frequency=100
    Noise Strength=&Multiply_Float(A=100, B=&FN_FrontEndFloatModifier(Gameplay=1, Front End=$System.FrontendScale))
    Pan Noise=false
   )
   &SolveForcesAndVelocity(Clamp Velocity=false, Limit Acceleration=false)
   &FN_Sprite_Attributes(
    Color=&MakeLinearColorFromVectorAndFloat(
     "Vector (RGB)"=&MakeVectorFromLinearColorRGB(Color=<rgba 1 0.6276 0 1>)
     "Float (Alpha)"=&FloatFromCurve(
      FloatCurve=DataInterface() {propertyValues {Curve {
         defaultValue=0
         keys [
          {interpMode=RCIM_Cubic, tangentMode=RCTM_User, time=0, value=1}
          {arriveTangent=-2.2441, interpMode=RCIM_Cubic, leaveTangent=-2.2441, tangentMode=RCTM_User, time=1, value=0}
         ]
        }}}
      Scale Curve=0.25
     )
    )
    Create Color Attribute=true
    Create SpriteSize Attribute=true
    SpriteSize=&Multiply_Vector2D_ByFloat(Float=2, Vector2D=$Particles.SpriteSize)
    Write to Module Namespace=true
   )
  ]
  rendererClasses [&NiagaraSpriteRendererProperties]
  renderers [
   {rendererClass=&NiagaraSpriteRendererProperties}
   {rendererClass=&NiagaraSpriteRendererProperties, rendererIndex=1}
  ]
 }
 Glow_n_Pop {
  emitterUpdateScript [
   &EmitterState()
   &SpawnBurst_Instantaneous(
    Spawn Time=&Subtract_Float(A=$System.IntroLength, B=0.15)
    Use Spawn Probability=false
    UseLoopCountLimit=false
   )
   &SpawnBurst_Instantaneous(Spawn Group=1, Spawn Time=0.3, Use Spawn Probability=false, UseLoopCountLimit=false) {module=SpawnBurst_Instantaneous001}
  ]
  particleSpawnScript [
   &InitializeParticle(
    Color=&MaskLinearColorBySpawnGroup(Masked Linear Color=<rgba 5 2.3856 0.525 0.8>, Passthrough Linear Color=<rgba 3 0.233 0.0781 1>)
    Lifetime=&MaskFloatBySpawnGroup(Masked Float=0.15, Passthrough Float=0.65)
    Position Offset=$System.BurstPosition
    Sprite Rotation Mode=::NewEnumerator0
    Sprite Size Mode=::NewEnumerator3
    Uniform Sprite Size=&Multiply_Float(
     A=&MaskFloatBySpawnGroup(Masked Float=18, Passthrough Float=30)
     B=&FN_FrontEndFloatModifier(Gameplay=1, Front End=$System.FrontendScale)
    )
    UsePositionOffset=true
    Write Material Random=true
   )
   &DynamicMaterialParameters(
    Index 0 Param 4=1
    Module.Index0Float0Write=true
    Module.Index0Float1Write=true
    Module.Index0Float2Write=true
    Module.Index0Float3Write=true
   )
   /SetVariables(Particles.VisibilityTag=&SpawnGroup()) {isSetParameters=true}
   &CameraOffset(Camera Offset Amount=&MaskFloatBySpawnGroup(Masked Float=0, Passthrough Float=10))
  ]
  particleUpdateScript [
   &ParticleState()
   &ScaleSpriteSize(
    Scale Sprite Size Mode=::NewEnumerator0
    Uniform Scale Factor=&FloatFromCurve(FloatCurve=DataInterface() {propertyValues {Curve {keys [
         {arriveTangent=3.3458, interpMode=RCIM_Cubic, leaveTangent=3.3458, tangentMode=RCTM_User, time=0, value=1}
         {arriveTangent=1.251, tangentMode=RCTM_User, time=1, value=3}
        ]}}})
   ) {module=ScaleSpriteSize002}
   &ScaleColor(
    Scale Alpha=&FloatFromCurve(FloatCurve=DataInterface() {propertyValues {Curve {
        defaultValue=0
        keys [
         {arriveTangent=0.3137, interpMode=RCIM_Cubic, leaveTangent=0.3137, tangentMode=RCTM_User, time=0, value=1}
         {arriveTangent=-2.0958, interpMode=RCIM_Cubic, leaveTangent=-2.0958, tangentMode=RCTM_User, time=1, value=0}
        ]
       }}})
    Scale RGB=&VectorFromCurve(VectorCurve=DataInterface() {propertyValues {
       XCurve {keys [{interpMode=RCIM_Cubic, tangentMode=RCTM_User, time=0, value=1}, {arriveTangent=-0.2294, tangentMode=RCTM_User, time=1, value=0}]}
       YCurve {keys [{interpMode=RCIM_Cubic, tangentMode=RCTM_User, time=0, value=1}, {arriveTangent=-0.2294, tangentMode=RCTM_User, time=1, value=0}]}
       ZCurve {keys [{interpMode=RCIM_Cubic, tangentMode=RCTM_User, time=0, value=1}, {arriveTangent=-0.2294, tangentMode=RCTM_User, time=1, value=0}]}
      }})
    ScaleA=true
    ScaleRGB=true
   )
   /SetVariables(Particles.Glow Color=&Multiply_LinearColorByFloat(
     Linear Color=&MultiplyLinearColor(Linear Color=$Particles.Color, Scale Factor=&MakeVector4_FromLinearColor(LinearColor=<rgba 1 0.6515 0 1>))
     Float=&FloatFromCurve(FloatCurve=DataInterface() {propertyValues {Curve {
         defaultValue=1
         keys [{interpMode=RCIM_Cubic, tangentMode=RCTM_User, time=0, value=1}, {interpMode=RCIM_Cubic, tangentMode=RCTM_User, time=1, value=0}]
        }}})
    )) {isSetParameters=true}
   &FN_Sprite_Attributes(
    Color=&MultiplyLinearColor(
     Linear Color=$Particles.Initial.Color
     Scale Factor=&MakeVector4FromVectorAndFloat(
      "Vector (RGB)"=<1 1 1>
      "Float (Alpha)"=&FloatFromCurve(FloatCurve=DataInterface() {propertyValues {Curve {
          defaultValue=1
          keys [
           {interpMode=RCIM_Cubic, tangentMode=RCTM_User, time=0.2, value=0}
           {arriveTangent=2.6956, interpMode=RCIM_Cubic, leaveTangent=2.6956, tangentMode=RCTM_User, time=0.975, value=1}
           {arriveTangent=-35.2676, interpMode=RCIM_Cubic, leaveTangent=-35.2676, tangentMode=RCTM_User, time=1, value=0}
          ]
         }}})
     )
    )
    Create Color Attribute=true
    Create SpriteSize Attribute=true
    SpriteSize=&Multiply_Vector2D_ByFloat(
     Float=&FloatFromCurve(FloatCurve=DataInterface() {propertyValues {Curve {
         defaultValue=1
         keys [{interpMode=RCIM_Cubic, tangentMode=RCTM_User, time=0, value=1}, {interpMode=RCIM_Cubic, tangentMode=RCTM_User, time=1, value=1.5}]
        }}})
     Vector2D=$Particles.Initial.SpriteSize
    )
    Write to Module Namespace=true
   )
  ]
  rendererClasses [&NiagaraSpriteRendererProperties]
  renderers [
   {rendererClass=&NiagaraSpriteRendererProperties}
   {rendererClass=&NiagaraSpriteRendererProperties, rendererIndex=1}
   {rendererClass=&NiagaraSpriteRendererProperties, rendererIndex=2}
  ]
 }
 Heart {
  emitterSpawnScript [
   &InitializeAudioPlayer(Audio Player=DataInterface() {propertyValues {
      Attenuation=null
      Concurrency=null
      MaxPlaysPerTick=10
      SoundToPlay=/CosmeticCompanions/Assets/Audio_Shared/Interaction/Petting/SC_Mimosa_Interaction_React_Default
      bAllowLoopingOneShotSounds=false
      bLimitPlaysPerTick=true
      bOnlyActiveDuringGameplay=false
      bStopWhenComponentIsDestroyed=true
      ConfigurationUserParameter {parameter {name=None, typeDefHandle {registeredTypeIndex=1}, varData []}}
      ParameterNames []
     }})
   /Niagara/Modules/Audio/PlayAudio() {module=PlayAudio001}
  ]
  emitterUpdateScript [
   &EmitterState()
   &SpawnBurst_Instantaneous(Use Spawn Probability=false, UseLoopCountLimit=false)
  ]
  particleSpawnScript [
   &InitializeParticle(
    Lifetime=$System.IntroLength
    Sprite Size Mode=::NewEnumerator3
    Uniform Sprite Size=&Multiply_Float(A=30, B=&FN_FrontEndFloatModifier(Gameplay=1, Front End=$System.FrontendScale))
    UsePositionOffset=false
    Write Material Random=true
   )
   &CameraOffset()
  ]
  particleUpdateScript [
   &ParticleState()
   &ScaleSpriteSize(
    Scale Sprite Size Mode=::NewEnumerator0
    Uniform Scale Factor=&FloatFromCurve(FloatCurve=DataInterface() {propertyValues {Curve {
        defaultValue=0
        keys [
         {arriveTangent=0.01, interpMode=RCIM_Cubic, leaveTangent=0.01, tangentMode=RCTM_User, time=0, value=2}
         {arriveTangent=0.0128, interpMode=RCIM_Cubic, leaveTangent=0.0128, tangentMode=RCTM_User, time=0.5, value=1}
         {arriveTangent=1.1199, interpMode=RCIM_Cubic, leaveTangent=1.1199, tangentMode=RCTM_User, time=0.6, value=1.1}
         {arriveTangent=-1.5678, interpMode=RCIM_Cubic, leaveTangent=-1.5678, tangentMode=RCTM_User, time=0.65, value=1.1}
         {arriveTangent=0.0128, interpMode=RCIM_Cubic, leaveTangent=0.0128, tangentMode=RCTM_User, time=0.7, value=1}
         {arriveTangent=1.4378, interpMode=RCIM_Cubic, leaveTangent=1.4378, tangentMode=RCTM_User, time=0.8, value=1.2}
         {arriveTangent=-2.0937, interpMode=RCIM_Cubic, leaveTangent=-2.0937, tangentMode=RCTM_User, time=0.85, value=1.2}
         {arriveTangent=-0.0175, interpMode=RCIM_Cubic, leaveTangent=-0.0175, tangentMode=RCTM_User, time=0.9, value=1}
         {arriveTangent=15.8361, interpMode=RCIM_Cubic, leaveTangent=15.8361, tangentMode=RCTM_User, time=1, value=2}
        ]
       }}})
   )
   &ScaleColor(
    Scale Alpha=&FloatFromCurve(
     FloatCurve=DataInterface() {propertyValues {Curve {
        defaultValue=0
        keys [
         <at 0.1 0 RCIM_Cubic>
         {arriveTangent=2.7505, interpMode=RCIM_Cubic, leaveTangent=-0.0041, tangentMode=RCTM_Break, time=0.7, value=1}
         {interpMode=RCIM_Cubic, tangentMode=RCTM_User, time=0.9, value=1}
         {arriveTangent=-20.1409, tangentMode=RCTM_User, time=1, value=0}
        ]
       }}}
     Scale Curve=0.8
    )
    ScaleA=true
    ScaleRGB=true
   )
   &DynamicMaterialParameters(
    Index 0 Param 1=1
    Index 0 Param 2=0.5
    Index 0 Param 3=&FN_FrontEndFloatModifier(
     Gameplay=&FNTimeOfDayScalar(Night Scale=0.4, Morning Scale=0.8, Evening Scale=0.8)
     Front End=0.5
    )
    Module.Index0Float0Write=true
    Module.Index0Float1Write=true
    Module.Index0Float2Write=true
    Module.Index0Float3Write=true
   )
   /SetVariables(
    Particles.Glow Color=&Multiply_LinearColorByFloat(
     Linear Color=<rgba 1 0.5 0 0.5>
     Float=&MakeFloatFromLinearColor(LinearColor=$Particles.Color, Channel:ENiagaraFloatFromLinearColorOptions=::NewEnumerator3)
    )
    Particles.Glow Size=&Multiply_Vector2D(A=$Particles.SpriteSize, B=<1.2 1.2>)
   ) {isSetParameters=true}
   /SetVariables(Particles.Position=&AddVectorToPosition(
     Position=Linked() {linkedVariable {name=System.BurstPosition, type=Vector3f}}
     Vector=&Multiply_VectorByFloat(
      Vector=&ReturnCameraVector(Camera Vectors=::NewEnumerator2, Coordinate Space=::Simulation)
      Float=&FloatFromCurve(
       FloatCurve=DataInterface() {propertyValues {Curve {
          defaultValue=1
          keys [
           {arriveTangent=-3.0007, interpMode=RCIM_Cubic, leaveTangent=-3.0007, tangentMode=RCTM_User, time=0, value=1}
           {interpMode=RCIM_Cubic, tangentMode=RCTM_User, time=0.7, value=0}
          ]
         }}}
       Scale Curve=-60
      )
     )
    )) {isSetParameters=true}
  ]
  rendererClasses [&NiagaraSpriteRendererProperties]
  renderers [{rendererClass=&NiagaraSpriteRendererProperties}]
 }
 Light {
  emitterUpdateScript [
   &EmitterState()
   &SpawnBurst_Instantaneous(Spawn Time=0.4, Use Spawn Probability=false, UseLoopCountLimit=false)
  ]
  particleSpawnScript [
   &InitializeParticle(Color=<rgba 3 0.233 0.0781 1>, Lifetime=0.6, Position Offset=$System.BurstPosition, UsePositionOffset=true, Write Material Random=true)
   &CameraOffset(Camera Offset Amount=10)
  ]
  particleUpdateScript [
   &ParticleState()
   &ScaleColor(
    Scale Alpha=&FloatFromCurve(FloatCurve=DataInterface() {propertyValues {Curve {
        defaultValue=1
        keys [
         {interpMode=RCIM_Cubic, tangentMode=RCTM_User, time=0.2, value=0}
         {arriveTangent=0.0314, interpMode=RCIM_Cubic, leaveTangent=0.0314, tangentMode=RCTM_User, time=0.975, value=1}
         {arriveTangent=-35.2676, interpMode=RCIM_Cubic, leaveTangent=-35.2676, tangentMode=RCTM_User, time=1, value=0}
        ]
       }}})
    ScaleA=true
    ScaleRGB=true
   )
  ]
  rendererClasses [&NiagaraLightRendererProperties]
  renderers [{rendererClass=&NiagaraLightRendererProperties, rendererIndex=0}]
 }
 Sparks {
  emitterUpdateScript [
   &EmitterState(Life Cycle Mode=::NewEnumerator1, Loop Behavior=::NewEnumerator1, Loop Duration=0.9, UseLoopDelay=false)
   &SpawnRate(
    SpawnRate=&FloatFromCurve(
     FloatCurve=DataInterface() {propertyValues {Curve {
        defaultValue=1
        keys [
         <at 0.3 0 RCIM_Constant>
         {interpMode=RCIM_Cubic, tangentMode=RCTM_User, time=0.65, value=0.2}
         {arriveTangent=4.8762, interpMode=RCIM_Cubic, leaveTangent=4.8762, tangentMode=RCTM_User, time=0.8, value=1}
        ]
       }}}
     CurveIndex=$Emitter.NormalizedLoopAge
     Scale Curve=150
    )
    Use Spawn Probability=false
   )
  ]
  particleSpawnScript [
   &FN_SetSpawnGroup(Number of Spawn Groups=3)
   &InitializeParticle(
    Lifetime Max=0.4
    Lifetime Min=0.2
    Lifetime Mode=::NewEnumerator1
    Position Offset=&Add_Vector(
     A=$System.BurstPosition
     B=&Add_Vector(
      A=&Multiply_VectorByFloat(
       Vector=&ReturnCameraVector(Camera Vectors=::NewEnumerator3, Coordinate Space=::Simulation)
       Float=&Multiply_Float(
        A=&FN_FrontEndFloatModifier(Gameplay=1, Front End=$System.FrontendScale)
        B=&MergeHelper_FloatSwitch(0=3, 1=-2.4, 2=0.3, 3=?, Index=&SpawnGroup(), Num=::NewEnumerator1)
       )
      )
      B=&Multiply_VectorByFloat(
       Vector=&ReturnCameraVector(Camera Vectors=::NewEnumerator2, Coordinate Space=::Simulation)
       Float=&Multiply_Float(
        A=&FN_FrontEndFloatModifier(Gameplay=1, Front End=$System.FrontendScale)
        B=&MergeHelper_FloatSwitch(0=-3, 1=-3, 2=3, 3=?, Index=&SpawnGroup(), Num=::NewEnumerator1)
       )
      )
     )
    )
    Sprite Size Max=<2.5 5>
    Sprite Size Min=<1.5 3>
    Sprite Size Mode=::NewEnumerator2
    UsePositionOffset=true
    Write Material Random=true
   )
   &CameraOffset(Camera Offset Amount=&FN_FrontEndFloatModifier(Gameplay=0, Front End=0))
   &ShapeLocation(
    MaskSpawnGroup=false
    Override Seed=false
    Sphere Radius=&Multiply_Float(A=1.2, B=&FN_FrontEndFloatModifier(Gameplay=1, Front End=$System.FrontendScale))
    Sphere Surface Distribution=1
    UseOwnerScale=true
   )
   &AddVelocity(
    Module.Use Falloff Distance=false
    Origin Offset=$System.BurstPosition
    Override Seed=false
    Velocity Mode=::NewEnumerator1
    Velocity Speed=&Multiply_Float(
     A=&RandomRangeFloat(Minimum=50, Maximum=150)
     B=&FN_FrontEndFloatModifier(Gameplay=1, Front End=$System.FrontendScale)
    )
   )
  ]
  particleUpdateScript [
   &ParticleState()
   &ScaleColor(
    Linear Color Curve=DataInterface() {propertyValues {
      AlphaCurve {keys [<at 0 1>,<at 1 0>]}
      BlueCurve {keys [<at 0 0.7043>,<at 0.7485 0.15>]}
      GreenCurve {keys [<at 0 1.4696>,<at 0.7485 0.3875>]}
      RedCurve {keys [<at 0 3>,<at 0.7485 3>]}
     }}
    Scale Mode=::NewEnumerator2
   )
   &ScaleSpriteSize(Scale Factor=&Vector2DFromFloat(Value=&FloatFromCurve(
      FloatCurve=DataInterface() {propertyValues {Curve {keys [
          {arriveTangent=10.1867, interpMode=RCIM_Cubic, leaveTangent=10.1867, tangentMode=RCTM_User, time=0, value=0}
          {interpMode=RCIM_Cubic, leaveTangent=-0.0263, tangentMode=RCTM_Break, time=0.2, value=1.2}
          {arriveTangent=-2.7631, interpMode=RCIM_Cubic, leaveTangent=-2.7631, tangentMode=RCTM_User, time=1, value=0}
         ]}}}
      Scale Curve=&Multiply_Float(A=0.24, B=&FN_FrontEndFloatModifier(Gameplay=1, Front End=$System.FrontendScale))
     ))) {module=ScaleSpriteSize001}
   &CurlNoiseForce(
    Noise Strength=&Multiply_Float(A=300, B=&FN_FrontEndFloatModifier(Gameplay=1, Front End=$System.FrontendScale))
    Pan Noise=true
    Pan Noise Field=<0 0 0.5>
    Random Seed=10
   )
   &Drag(
    Drag=&Multiply_Float(A=6, B=&FN_FrontEndFloatModifier(Gameplay=1, Front End=$System.FrontendScale))
    Use Linear Drag=true
    Use Rotational Drag=false
   )
   &SolveForcesAndVelocity(Clamp Velocity=false, Limit Acceleration=false)
   &ScaleSpriteSizeBySpeed(Max Scale Factor=<0.8 2>, Velocity Threshold=100)
   /SetVariables(Particles.Glow03_Color=&MakeLinearColorFromVectorAndFloat(
     "Vector (RGB)"=&MakeVectorFromLinearColorRGB(Color=$Particles.Color)
     "Float (Alpha)"=&Multiply_Float(A=0.2, B=1)
    )) {isSetParameters=true}
   /SetVariables(Particles.Glow03_Size=&Multiply_Vector2D(A=$Particles.SpriteSize, B=<4 2>)) {isSetParameters=true}
  ]
  rendererClasses [&NiagaraSpriteRendererProperties]
  renderers [
   {rendererClass=&NiagaraSpriteRendererProperties}
   {rendererClass=&NiagaraSpriteRendererProperties, rendererIndex=1}
  ]
 }
}
summary {
 systemName=NS_ReactFX_WindTyro
 emitters [
  {emitterName=Heart, rendererClasses [&NiagaraSpriteRendererProperties]}
  {emitterName=Cogs, rendererClasses [&NiagaraSpriteRendererProperties, &NiagaraMeshRendererProperties]}
  {emitterName=Sparks, rendererClasses [&NiagaraSpriteRendererProperties]}
  {emitterName=Light, rendererClasses [&NiagaraLightRendererProperties]}
  {emitterName=CrossStars, rendererClasses [&NiagaraSpriteRendererProperties]}
  {emitterName=Glow_n_Pop, rendererClasses [&NiagaraSpriteRendererProperties]}
 ]
 userVariables [
  User.bIsFrontEnd {type=NiagaraBool}
  User.bIsFrontEndPreview {type=NiagaraBool}
 ]
}
