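# Niagara reaction effect for the WittyStork cosmetic companion, playing a shared petting-reaction sound with bubble-like sprites.
# Intro emits six staggered bursts that fly outward, slow under drag, then swell and fade, with curve-driven pop and glow attributes.
# WriteDeath fills a position array, which system-level ReadDeath appears to turn into small bubble sprays that rise at death positions.
# Global scale sizes particles, and offsets differ between gameplay and front end.
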NiagaraSystem /CosmeticCompanions/FX/React/NS_ReactFX_WittyStork
dependencies {
 usedDataInterfaces [
  &NiagaraDataInterfaceAudioPlayer
  &NiagaraDataInterfaceCurve
  &NiagaraDataInterfaceColorCurve
  &NiagaraDataInterfaceArrayPosition
 ]
 usedDynamicInputs [&ReturnMeshOrientationAxis, &FN_FrontEndFloatModifier]
 usedModules [
  &self:SystemSpawnScript.SystemScriptSource.SystemScriptGraph.NiagaraNodeAssignment_0.SetVariables_#1
  &self:SystemSpawnScript.SystemScriptSource.SystemScriptGraph.NiagaraNodeAssignment_5.SetVariables_#2
  &self:SystemSpawnScript.SystemScriptSource.SystemScriptGraph.NiagaraNodeAssignment_2.SetVariables_#3
  &self:SystemSpawnScript.SystemScriptSource.SystemScriptGraph.NiagaraNodeAssignment_6.SetVariables_#4
  &self:SystemSpawnScript.SystemScriptSource.SystemScriptGraph.NiagaraNodeAssignment_1.SetVariables_#5
  &SystemState
  &self:ReadDeath
  &self:SystemSpawnScript.SystemScriptSource.SystemScriptGraph.NiagaraNodeAssignment_7.SetVariables_#6
 ]
}
emitters {
 Bubbles {
  emitterUpdateScript [
   &EmitterState(
    Life Cycle Mode=::NewEnumerator1
    Loop Behavior=::NewEnumerator1
    Loop Duration=&Add_Float(A=$System.IntroLength, B=1)
    UseLoopDelay=false
   )
   &SpawnPerFrame(
    Spawn:NiagaraBool=$System.DeathSpawn
    Spawn Count:NiagaraInt32=&RandomRangeInt(Minimum=4, Maximum=6)
    Use Spawn Probability=false
   )
  ]
  particleSpawnScript [
   &InitializeParticle(
    Lifetime=&RandomRangeFloat(Minimum=0.5, Maximum=2)
    Position Offset=<0 0 100>
    Sprite Rotation Mode=::NewEnumerator2
    Sprite Size Mode=::NewEnumerator3
    Uniform Sprite Size=&Multiply_Float(A=$System.GlobalScale, B=&RandomRangeFloat(Minimum=5, Maximum=5))
    UsePositionOffset=false
    Write Material Random=true
   )
   &ShapeLocation(MaskSpawnGroup=false, Override Seed=false, Shape Origin=$System.DeathPosition, Sphere Radius=5, UseOwnerScale=true)
   &AddVelocity(
    Velocity=&RandomRangeVector(Minimum=<-1 -1 -1>, Maximum=<1 1 1>)
    Velocity Speed Scale=&Multiply_Float(A=&RandomRangeFloat(Minimum=50, Maximum=100), B=$System.GlobalScale)
   )
  ]
  particleUpdateScript [
   &ParticleState()
   &ScaleSpriteSize(
    Scale Sprite Size Mode=::NewEnumerator0
    Uniform Scale Factor=&FloatFromCurve(FloatCurve=DataInterface() {propertyValues {Curve {keys [
         {arriveTangent=7.8562, interpMode=RCIM_Cubic, leaveTangent=7.8562, tangentMode=RCTM_User, time=0, value=0.25}
         {arriveTangent=0.1667, interpMode=RCIM_Cubic, leaveTangent=0.1667, time=0.2487, value=1.1246}
         {arriveTangent=-1.497, interpMode=RCIM_Cubic, leaveTangent=-1.497, time=0.6, value=0.35}
         <at 1 0 RCIM_Cubic>
        ]}}})
   )
   &DynamicMaterialParameters(
    Index 0 Param 1=5
    Index 0 Param 2=12
    Index 0 Param 3=&FloatFromCurve(FloatCurve=DataInterface() {propertyValues {Curve {keys [<at 0 5 RCIM_Cubic>, {arriveTangent=-4.1151, interpMode=RCIM_Cubic, leaveTangent=-4.1151, tangentMode=RCTM_User, time=1, value=1}]}}})
    Module.Param0WriteEnabled=true
    Module.Param1WriteEnabled=true
    Module.Param2WriteEnabled=true
    Module.Param3WriteEnabled=false
   )
   &CurlNoiseForce(Bias Noise=false, Noise Strength=50, Pan Noise=false)
   &Drag(Drag=&RandomRangeFloat(Minimum=3, Maximum=4), Use Linear Drag=true, Use Rotational Drag=true)
   &GravityForce(Gravity=<0 0 55>)
   &SolveForcesAndVelocity(Clamp Velocity=false, Limit Acceleration=false)
   &ScaleColor(Scale RGB=&MakeVectorFromLinearColorRGB(Color=<rgba 1 1 1 1>), ScaleA=true, ScaleRGB=true)
  ]
  rendererClasses [&NiagaraSpriteRendererProperties]
  renderers [{rendererClass=&NiagaraSpriteRendererProperties}]
 }
 Intro {
  emitterSpawnScript [
   &InitializeAudioPlayer(Audio Player=DataInterface() {propertyValues {
      Attenuation=null
      Concurrency=null
      MaxPlaysPerTick=10
      SoundToPlay=/CosmeticCompanions/Assets/Audio_Shared/Interaction/Petting/SC_Mimosa_Interaction_React_Default
      bAllowLoopingOneShotSounds=false
      bLimitPlaysPerTick=true
      bOnlyActiveDuringGameplay=false
      bStopWhenComponentIsDestroyed=true
      ConfigurationUserParameter {parameter {name=None, typeDefHandle {registeredTypeIndex=1}, varData []}}
      ParameterNames []
     }})
   /Niagara/Modules/Audio/PlayAudio() {module=PlayAudio001}
  ]
  emitterUpdateScript [
   &EmitterState()
   &SpawnBurst_Instantaneous(Use Spawn Probability=false, UseLoopCountLimit=false)
   &SpawnBurst_Instantaneous(Spawn Time=0.03, Use Spawn Probability=false, UseLoopCountLimit=false) {module=SpawnBurst_Instantaneous001}
   &SpawnBurst_Instantaneous(Spawn Time=0.06, Use Spawn Probability=false, UseLoopCountLimit=false) {module=SpawnBurst_Instantaneous002}
   &SpawnBurst_Instantaneous(Spawn Time=0.09, Use Spawn Probability=false, UseLoopCountLimit=false) {module=SpawnBurst_Instantaneous003}
   &SpawnBurst_Instantaneous(Spawn Time=0.12, Use Spawn Probability=false, UseLoopCountLimit=false) {module=SpawnBurst_Instantaneous004}
   &SpawnBurst_Instantaneous(Spawn Time=0.15, Use Spawn Probability=false, UseLoopCountLimit=false) {module=SpawnBurst_Instantaneous005}
  ]
  particleSpawnScript [
   /SetVariables(Particles.VisibilityTag=$Emitter.SpawnGroup) {isSetParameters=true}
   /SetVariables(Particles.Lifetime=&Add_Float(A=$System.IntroLength, B=&RandomRangeFloat(Minimum=-0.1, Maximum=0.2))) {isSetParameters=true}
   &InitializeParticle(
    Color=<rgba 0 352.0194 500 1>
    Lifetime=$Particles.Lifetime
    Position Offset=&FN_FrontEndVector3Modifier(Gameplay=<-7.5 15 70>, Front End=<-7.5 15 60>)
    Ribbon Width=10
    Ribbon Width Mode=::NewEnumerator1
    Sprite Rotation Mode=::NewEnumerator0
    Sprite Size Mode=::NewEnumerator3
    Uniform Sprite Size=&Multiply_Float(A=&RandomRangeFloat(Minimum=10, Maximum=12.5), B=$System.GlobalScale)
    UsePositionOffset=true
    Write Material Random=true
   ) {enabled=false}
   &InitializeParticle(
    Color=<rgba 0 352.0194 500 1>
    Lifetime=$Particles.Lifetime
    Position Offset=&FN_FrontEndVector3Modifier(Gameplay=<0 0 -50>, Front End=<0 0 -70>)
    Ribbon Width=10
    Ribbon Width Mode=::NewEnumerator1
    Sprite Rotation Mode=::NewEnumerator0
    Sprite Size Mode=::NewEnumerator3
    Uniform Sprite Size=&Multiply_Float(A=&RandomRangeFloat(Minimum=10, Maximum=12.5), B=$System.GlobalScale)
    UsePositionOffset=true
    Write Material Random=true
   ) {module=InitializeParticle001}
   &ShapeLocation(
    Handle Radius=5
    Large Radius=33.8941
    MaskSpawnGroup=false
    Override Seed=false
    Rotation Mode=::NewEnumerator2
    Shape Primitive=::NewEnumerator3
    Surface Distribution=1
    U Distribution=0.75
    UseOwnerScale=true
    Yaw / Pitch / Roll=<0 0.115 0.25>
   )
   &AddVelocity(
    Velocity=&AddVectorToPosition(
     Position=<0 0 0>
     Vector=&NormalizeVector(Vector To Normalize=$Particles.ShapeLocation.ShapeVector, Safe Normalize=true, Safe Fallback Threshold=0, Safe Fallback Vector=<0 0 0.001>)
    )
    Velocity Speed Scale=&Multiply_Float(A=&RandomRangeFloat(Minimum=100, Maximum=300), B=$System.GlobalScale)
   ) {module=AddVelocity001}
   /SetVariables(
    Particles.PopColor:LinearColor=<rgba 5 5 5 0.8>
    Particles.PopLifetime:NiagaraFloat=&Add_Float(A=$Particles.Lifetime, B=0)
    Particles.PopSize:Vector2f=&Vector2DFromFloat(Value=&Multiply_Float(A=10, B=$System.GlobalScale))
   ) {isSetParameters=true}
  ]
  particleUpdateScript [
   &ParticleState()
   &Drag(Drag=&RandomRangeFloat(Minimum=5, Maximum=7), Use Linear Drag=true, Use Rotational Drag=true)
   &SolveForcesAndVelocity(Clamp Velocity=false, Limit Acceleration=false)
   &self:WriteDeath(Position Array:NiagaraDataInterfaceArrayPosition=$System.DeathArray) {module=WriteDeath}
   &ScaleSpriteSize(
    Scale Sprite Size Mode=::NewEnumerator0
    Uniform Scale Factor=&Add_Float(
     A=0
     B=&FloatFromCurve(FloatCurve=DataInterface() {propertyValues {Curve {
         defaultValue=0
         keys [
          <at 0 0.3 RCIM_Cubic>
          {arriveTangent=1.25, leaveTangent=1.25, time=0.1915, value=0.9}
          {arriveTangent=5.5087, interpMode=RCIM_Cubic, leaveTangent=5.5087, tangentMode=RCTM_User, time=0.6915, value=1}
          {arriveTangent=26.9573, interpMode=RCIM_Cubic, leaveTangent=26.9573, tangentMode=RCTM_User, time=0.7915, value=2.25}
         ]
        }}})
    )
   ) {module=ScaleSpriteSize002}
   &ScaleColor(
    Scale Alpha=&FloatFromCurve(FloatCurve=DataInterface() {propertyValues {Curve {
        defaultValue=0
        keys [
         {arriveTangent=9.0725, interpMode=RCIM_Cubic, leaveTangent=9.0725, tangentMode=RCTM_User, time=0, value=0}
         {arriveTangent=-0.2441, leaveTangent=-0.2441, time=0.1408, value=0.9847}
         {arriveTangent=-1.146, leaveTangent=-1.146, time=0.7602, value=1}
         {arriveTangent=-0.0405, arriveTangentWeight=0.0167, interpMode=RCIM_Cubic, leaveTangent=-0.0405, tangentMode=RCTM_User, time=0.8102, value=0}
        ]
       }}})
    ScaleA=true
    ScaleRGB=true
   )
   &DynamicMaterialParameters(
    Index 0 Param 1=&FloatFromCurve(
     FloatCurve=DataInterface() {propertyValues {Curve {keys [{arriveTangent=0.6296, interpMode=RCIM_Cubic, leaveTangent=0.6296, tangentMode=RCTM_User, time=0.5, value=1}, <at 0.6 50>]}}}
     Scale Curve=5
    )
    Index 0 Param 2=&FloatFromCurve(FloatCurve=DataInterface() {propertyValues {Curve {keys [{arriveTangent=0.6296, interpMode=RCIM_Cubic, leaveTangent=0.6296, tangentMode=RCTM_User, time=0.5, value=1}, <at 0.6 50>]}}})
    Index 0 Param 3=0.2
    Module.Index0Float0Write=true
    Module.Index0Float1Write=true
    Module.Index0Float2Write=true
    Module.Index0Float3Write=true
   )
   /SetVariables(Particles.UpdatePopAge:NiagaraFloat=&Divide_Float(A=$Particles.Age, B=$Particles.Initial.PopLifetime)) {isSetParameters=true}
   /SetVariables(
    Particles.PopDynParam:Vector4f=&MakeVector4(X=0, Y=0, Z=0, W=1)
    Particles.UpdatePopColor:LinearColor=&MultiplyLinearColor(
     Linear Color=$Particles.PopColor
     Scale Factor=&MakeVector4FromVectorAndFloat(
      "Vector (RGB)"=&MakeVectorFromLinearColorRGB(Color=&ColorFromCurve(
        ColorCurve=DataInterface() {propertyValues {
          AlphaCurve {keys [<at 0 1 RCIM_Cubic>]}
          BlueCurve {keys [<at 0.85 0 RCIM_Cubic>,<at 1 1 RCIM_Cubic>]}
          GreenCurve {keys [<at 0.85 0 RCIM_Cubic>,<at 1 1 RCIM_Cubic>]}
          RedCurve {keys [<at 0.85 1 RCIM_Cubic>,<at 1 1 RCIM_Cubic>]}
         }}
        CurveIndex=$Particles.UpdatePopAge
       ))
      "Float (Alpha)"=&FloatFromCurve(
       FloatCurve=DataInterface() {propertyValues {Curve {
          defaultValue=0
          keys [
           {arriveTangent=-2.5882, interpMode=RCIM_Cubic, time=0.725, value=0}
           <at 0.8 0.9059 RCIM_Cubic>
           {arriveTangent=-5.9285, interpMode=RCIM_Cubic, leaveTangent=-5.9285, tangentMode=RCTM_User, time=1, value=0}
          ]
         }}}
       CurveIndex=$Particles.UpdatePopAge
       Scale Curve=3
      )
     )
    )
    Particles.UpdatePopSize:Vector2f=&Multiply_Vector2D_ByFloat(
     Float=&FloatFromCurve(
      FloatCurve=DataInterface() {propertyValues {Curve {
         defaultValue=0
         keys [<at 0.725 1 RCIM_Cubic>, {arriveTangent=6.0193, interpMode=RCIM_Cubic, leaveTangent=6.0193, tangentMode=RCTM_User, time=1, value=3}]
        }}}
      CurveIndex=$Particles.UpdatePopAge
     )
     Vector2D=$Particles.PopSize
    )
   ) {isSetParameters=true}
   /SetVariables(
    Particles.Glow Color=&Multiply_LinearColorByFloat(
     Linear Color=&MultiplyLinearColor(
      Linear Color=$Particles.UpdatePopColor
      Scale Factor=&MakeVector4_FromLinearColor(LinearColor=<rgba 1 1 1 0.5>)
     )
     Float=0.125
    )
    Particles.Glow Size=&Multiply_Vector2D(A=$Particles.UpdatePopSize, B=<1 1>)
   ) {isSetParameters=true}
  ]
  rendererClasses [&NiagaraSpriteRendererProperties]
  renderers [
   {rendererClass=&NiagaraSpriteRendererProperties}
   {rendererClass=&NiagaraSpriteRendererProperties, rendererIndex=1}
   {rendererClass=&NiagaraSpriteRendererProperties, rendererIndex=2}
  ]
 }
}
summary {
 systemName=NS_ReactFX_WittyStork
 emitters [
  {emitterName=Intro, rendererClasses [&NiagaraSpriteRendererProperties]}
  {emitterName=Bubbles, rendererClasses [&NiagaraSpriteRendererProperties]}
 ]
 userVariables [
  User.bIsFrontEnd {type=NiagaraBool}
  User.bIsFrontEndPreview {type=NiagaraBool}
  User.FrontEndScale {type=NiagaraFloat, defaultValue=0.6, description=""}
 ]
}
