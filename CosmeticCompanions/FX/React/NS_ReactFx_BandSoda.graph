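# NS_ReactFx_BandSoda is a Niagara cosmetic-companion reaction effect: a soda-blob burst with a petting sound.
# Metaball1 bursts four particles after an intro delay, packs their blob and capsule data, reads it back through a particle attribute reader, and checks whether the camera is inside its bounds.
# FallBack mirrors that drag-and-curl-noise motion with simpler time-of-day-tinted meshes.
# Droplets and FallBackHeart carry no renderers.
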NiagaraSystem /CosmeticCompanions/FX/React/NS_ReactFx_BandSoda
dependencies {
 usedDataInterfaces [
  &NiagaraDataInterfaceAudioPlayer
  &NiagaraDataInterfaceParticleRead
  &NiagaraDataInterfaceCamera
  &NiagaraDataInterfaceCollisionQuery
  &NiagaraDataInterfaceCurve
  &NiagaraDataInterfaceColorCurve
 ]
 usedDynamicInputs [
  &Divide_Float
  /Niagara/DynamicInputs/Math/Floor_Float
  &MakeCustomFloatFromBool
  &SetBoolByIntComparison
  &SpawnGroup
  &FN_FrontEndVector3Modifier
  &Multiply_FloatByInt
 ]
 usedModules [
  &self:SystemSpawnScript.SystemScriptSource.SystemScriptGraph.NiagaraNodeAssignment_4.SetVariables_#1
  &self:SystemSpawnScript.SystemScriptSource.SystemScriptGraph.NiagaraNodeAssignment_6.SetVariables_#2
  &self:SystemSpawnScript.SystemScriptSource.SystemScriptGraph.NiagaraNodeAssignment_9.SetVariables_#3
  &self:SystemSpawnScript.SystemScriptSource.SystemScriptGraph.NiagaraNodeAssignment_10.SetVariables_#4
  &self:SystemSpawnScript.SystemScriptSource.SystemScriptGraph.NiagaraNodeAssignment_5.SetVariables_#5
  &SystemState
  &self:SystemSpawnScript.SystemScriptSource.SystemScriptGraph.NiagaraNodeAssignment_1.SetVariables_#6
  &self:SystemSpawnScript.SystemScriptSource.SystemScriptGraph.NiagaraNodeAssignment_2.SetVariables_#7
  &self:SystemSpawnScript.SystemScriptSource.SystemScriptGraph.NiagaraNodeAssignment_0.SetVariables_#8
  &self:SystemSpawnScript.SystemScriptSource.SystemScriptGraph.NiagaraNodeAssignment_3.SetVariables_#9
 ]
}
emitters {
 Droplets {rendererClasses [], renderers []}
 FallBack {
  emitterSpawnScript [
   &InitializeAudioPlayer(Audio Player=DataInterface() {propertyValues {
      Attenuation=null
      Concurrency=null
      MaxPlaysPerTick=10
      SoundToPlay=/CosmeticCompanions/Assets/Audio_Shared/Interaction/Petting/SC_Mimosa_Interaction_React_Default
      bAllowLoopingOneShotSounds=false
      bLimitPlaysPerTick=true
      bOnlyActiveDuringGameplay=false
      bStopWhenComponentIsDestroyed=true
      ConfigurationUserParameter {parameter {name=None, typeDefHandle {registeredTypeIndex=1}, varData []}}
      ParameterNames []
     }})
   /Niagara/Modules/Audio/SetInitialAudioParameter()
  ]
  emitterUpdateScript [
   &EmitterState()
   &SpawnBurst_Instantaneous(Spawn Count=4, Spawn Time=$System.IntroLength, Use Spawn Probability=false, UseLoopCountLimit=false)
  ]
  particleSpawnScript [
   &InitializeParticle(
    Color=&MultiplyLinearColor(
     Linear Color=<rgba 1 1 1 1>
     Scale Factor=&MakeVector4_FromLinearColor(LinearColor=&FNTimeOfDayColor(
       Night Color=<rgba 0.5967 0.6124 0.7344 1>
       Morning Color=<rgba 0.9115 0.7506 0.687 1>
       Day Color=<rgba 1 1 1 1>
       Evening Color=<rgba 0.9479 0.8252 0.7652 1>
      ))
    )
    Lifetime=0.45
    Mesh Scale Mode=::NewEnumerator3
    Mesh Uniform Scale=&MergeHelper_FloatSwitch(
     0=2.4
     1=2.4
     2=2.7
     3=1.3
     Index=&ReturnParticleIDAttributes(Particle ID Attribute=::NewEnumerator1)
     Num=::NewEnumerator2
    )
    Sprite Size Mode=::NewEnumerator3
    Uniform Sprite Size=&MergeHelper_FloatSwitch(
     0=2.4
     1=2.4
     2=2.7
     3=1.3
     Index=&ReturnParticleIDAttributes(Particle ID Attribute=::NewEnumerator1)
     Num=::NewEnumerator2
    )
    UsePositionOffset=false
    Write Material Random=true
   )
   &self:TraceStartPos(DownOffset=128) {enabled=false, module=TraceStartPos}
   &ShapeLocation(
    MaskSpawnGroup=false
    Offset=&AddVectorToPosition(Position=Linked() {linkedVariable {name=System.BurstPosition, type=Vector3f}}, Vector=<0 0 -30>)
    Offset Mode=::NewEnumerator0
    Override Seed=false
    Sphere Radius=0
    UseOwnerScale=true
   )
   &AddVelocity(Velocity=&Multiply_VectorByFloat(Vector=&RandomRangeVector(Minimum=<-1 -1 0>, Maximum=<1 1 0>), Float=500))
  ]
  particleUpdateScript [
   &ParticleState()
   &ScaleMeshSize(Scale Factor=&VectorFromFloat(Value=&FloatFromCurve(
      FloatCurve=DataInterface() {propertyValues {Curve {
         defaultValue=1
         keys [
          {arriveTangent=0.0679, interpMode=RCIM_Cubic, leaveTangent=0.0679, tangentMode=RCTM_User, time=0.8, value=1}
          {arriveTangent=-10.2246, interpMode=RCIM_Cubic, leaveTangent=-10.2246, tangentMode=RCTM_User, time=1, value=0}
         ]
        }}}
      Scale Curve=0.03
     )))
   &self:ScratchModule_03(
    CamVec:Vector3f=&ReturnCameraVector(Camera Vectors=::NewEnumerator4, Coordinate Space=::Simulation, Origin Position=$Particles.Position)
    CamVecRight:Vector3f=&ReturnCameraVector(Camera Vectors=::NewEnumerator3, Coordinate Space=::Simulation)
    MulltVel=2
    p0:NiagaraPosition=<-2.8 0 3.5>
    p1:NiagaraPosition=<2.8 0 3.5>
    p2:NiagaraPosition=<0 0 0.8>
    p3:NiagaraPosition=<0 0 -2>
   ) {module=ScratchModule_03}
   &Drag(Drag=14, Use Linear Drag=true, Use Rotational Drag=true)
   &CurlNoiseForce(Bias Noise=false, Noise Strength=65, Pan Noise=true, Pan Noise Field=<0 0 2>)
   &SolveForcesAndVelocity(Clamp Velocity=false, Limit Acceleration=false)
   /SetVariables(
    Particles.PriteSizeForSoftGlow:Vector2f=&Multiply_Vector2D(A=$Particles.SpriteSize, B=<4.5 4.5>)
    Particles.SpriteColor=&MultiplyLinearColor(
     Linear Color=<rgba 0.0314 0.3248 0.4115 1>
     Scale Factor=&MakeVector4_FromLinearColor(LinearColor=&FNTimeOfDayColor(
       Night Color=<rgba 0.1153 0.5677 0.538 1>
       Morning Color=<rgba 0.9115 0.7506 0.687 1>
       Day Color=<rgba 1 1 1 1>
       Evening Color=<rgba 0.9479 0.8252 0.7652 1>
      ))
    )
   ) {isSetParameters=true}
  ]
  rendererClasses [&NiagaraMeshRendererProperties]
  renderers [{rendererClass=&NiagaraMeshRendererProperties, rendererIndex=0}]
 }
 FallBackHeart {rendererClasses [], renderers []}
 Metaball1 {
  emitterSpawnScript [
   &InitializeAudioPlayer(Audio Player=DataInterface() {propertyValues {
      Attenuation=null
      Concurrency=null
      MaxPlaysPerTick=10
      SoundToPlay=&SC_Mimosa_Interaction_React_Default
      bAllowLoopingOneShotSounds=false
      bLimitPlaysPerTick=true
      bOnlyActiveDuringGameplay=false
      bStopWhenComponentIsDestroyed=true
      ConfigurationUserParameter {parameter {name=None, typeDefHandle {registeredTypeIndex=1}, varData []}}
      ParameterNames []
     }})
   &SetInitialAudioParameter()
   /Niagara/Modules/Audio/PlayAudio()
   /SetVariables(Emitter.ParticleAttribute=DataInterface() {propertyValues {EmitterBinding {bindingMode=Other, emitterName=Metaball1}}}) {isSetParameters=true}
  ]
  emitterUpdateScript [
   &EmitterState()
   &SpawnBurst_Instantaneous(Spawn Count=4, Spawn Time=$System.IntroLength, Use Spawn Probability=false, UseLoopCountLimit=false)
   /SetVariables(Emitter.Offset:Vector3f=$System.BurstPosition) {isSetParameters=true}
   /SetVariables(Emitter.Offset_2:Vector3f=<5 -20 0>) {enabled=false, isSetParameters=true}
   /SetVariables(Emitter.Bound:Vector3f=&Divide_Vector(A:Vector3f=<50 50 100>, B:Vector3f=<100 100 100>)) {isSetParameters=true}
   /SetVariables(
    Emitter.Color:LinearColor=&Multiply_LinearColorByFloat(
     Linear Color=&FN_Select_Color_from_Time_Of_Day(Morning=<rgba 0.8698 0.7989 0.7835 1>, Day=<rgba 1 1 1 1>, Evening=<rgba 0.9062 0.8324 0.8163 1>, Night=<rgba 0.6146 0.6146 0.6146 1>)
     Float=&Power(a=$System.TotalScale, b=0.5)
    )
    Emitter.Position:NiagaraPosition=&AddVectorToPosition(Position=$Engine.Emitter.SimulationPosition, Vector=$Emitter.Offset)
    Emitter.Scale:Vector3f=<1 1 1>
   ) {isSetParameters=true}
   /SetVariables(Emitter.Position_2:NiagaraPosition=&AddVectorToPosition(Position=$Engine.Emitter.SimulationPosition, Vector=$Emitter.Offset_2)) {enabled=false, isSetParameters=true}
   &self:ReadMySelf(Particle Attribute Reader:NiagaraDataInterfaceParticleRead=$Emitter.ParticleAttribute) {module=ReadMySelf}
   &self:ReadMySelf(Particle Attribute Reader:NiagaraDataInterfaceParticleRead=$Emitter.ParticleAttribute, Particle Index=1) {module=ReadMySelf001}
   &self:ReadMySelf(Particle Attribute Reader:NiagaraDataInterfaceParticleRead=$Emitter.ParticleAttribute, Particle Index=2) {module=ReadMySelf002}
   &self:ReadMySelf(Particle Attribute Reader:NiagaraDataInterfaceParticleRead=$Emitter.ParticleAttribute, Particle Index=3) {module=ReadMySelf003}
   &self:ReadMySelf(Particle Attribute Reader:NiagaraDataInterfaceParticleRead=$Emitter.ParticleAttribute, Particle Index=4) {enabled=false, module=ReadMySelf004}
   &self:ReadMySelf(Particle Attribute Reader:NiagaraDataInterfaceParticleRead=$Emitter.ParticleAttribute, Particle Index=1) {enabled=false, module=ReadMySelf005}
   &self:ReadMySelf(Particle Attribute Reader:NiagaraDataInterfaceParticleRead=$Emitter.ParticleAttribute, Particle Index=3) {enabled=false, module=ReadMySelf006}
   &self:ReadMySelf(Particle Attribute Reader:NiagaraDataInterfaceParticleRead=$Emitter.ParticleAttribute, Particle Index=5) {enabled=false, module=ReadMySelf007}
   &self:ReadMySelf(Particle Attribute Reader:NiagaraDataInterfaceParticleRead=$Emitter.ParticleAttribute, Particle Index=7) {enabled=false, module=ReadMySelf008}
   &self:ReadMySelf(Particle Attribute Reader:NiagaraDataInterfaceParticleRead=$Emitter.ParticleAttribute, Particle Index=9) {enabled=false, module=ReadMySelf009}
   &self:IsCameraInsideBox(
    Bound:Vector3f=&Multiply_VectorByFloat(Vector=$Emitter.Bound, Float=100)
    Box Origin:Vector3f=$Emitter.Offset
    Camera Query=DataInterface() {propertyValues {PlayerControllerIndex=0, bRequireCurrentFrameData=false}}
   ) {module=IsCameraInsideBox}
   &self:IsCameraInsideBox(
    Bound:Vector3f=&Multiply_VectorByFloat(Vector=$Emitter.Bound, Float=100)
    Box Origin:Vector3f=$Emitter.Offset_2
    Camera Query=DataInterface() {propertyValues {PlayerControllerIndex=0, bRequireCurrentFrameData=false}}
   ) {enabled=false, module=IsCameraInsideBox001}
  ]
  particleSpawnScript [
   /SetVariables(Particles.RDM:NiagaraFloat=&RandomRangeFloat(Minimum=0, Maximum=1)) {enabled=false, isSetParameters=true}
   &InitializeParticle(
    Sprite Size Mode=::NewEnumerator3
    Uniform Sprite Size=&MergeHelper_FloatSwitch(
     0=2
     1=2
     2=1.5
     3=-2
     Index=&ReturnParticleIDAttributes(Particle ID Attribute=::NewEnumerator1)
     Num=::NewEnumerator2
    )
    UsePositionOffset=false
    Write Material Random=true
   )
   &ShapeLocation(
    MaskSpawnGroup=false
    Offset=&AddVectorToPosition(Position=Linked() {linkedVariable {name=System.BurstPosition, type=Vector3f}}, Vector=<0 0 -30>)
    Offset Mode=::NewEnumerator0
    Override Seed=false
    Sphere Radius=0
    UseOwnerScale=true
   )
   &AddVelocity(Velocity=&Multiply_VectorByFloat(Vector=&RandomRangeVector(Minimum=<-1 -1 0>, Maximum=<1 1 0>), Float=500))
   &self:TraceStartPos(DisableTrace=true, DownOffset=128) {enabled=false, module=TraceStartPos}
  ]
  particleUpdateScript [
   &ParticleState()
   &Timeline(
    Initial Start Time:NiagaraFloat=&RandomRangeFloat(Minimum=10, Maximum=100)
    Play Rate=&FloatFromCurve(FloatCurve=DataInterface() {propertyValues {Curve {keys [<at 0.3968 0.4721>,<at 0.4686 1>]}}})
    Timeline End=999999
   ) {enabled=false}
   &ScaleSpriteSize(
    Uniform Curve Scale=1.1
    Uniform Curve Sprite Scale=DataInterface() {propertyValues {Curve {
       defaultValue=1
       keys [
        {arriveTangent=0.0679, interpMode=RCIM_Cubic, leaveTangent=0.0679, tangentMode=RCTM_User, time=0.7729, value=1}
        {arriveTangent=20.5532, interpMode=RCIM_Cubic, leaveTangent=20.5532, tangentMode=RCTM_User, time=1, value=3}
       ]
      }}}
   )
   &self:ScratchModule_03(
    CamVec:Vector3f=&ReturnCameraVector(Camera Vectors=::NewEnumerator4, Coordinate Space=::Simulation, Origin Position=$Particles.Position)
    CamVecRight:Vector3f=&ReturnCameraVector(Camera Vectors=::NewEnumerator3, Coordinate Space=::Simulation)
    MulltVel=2
   ) {module=ScratchModule_03}
   &Drag(Drag=14, Use Linear Drag=true, Use Rotational Drag=true)
   &CurlNoiseForce(Bias Noise=false, Noise Strength=65, Pan Noise=true, Pan Noise Field=<0 0 2>)
   &SolveForcesAndVelocity(Clamp Velocity=true, Limit Acceleration=false)
   /SetVariables() {isSetParameters=true}
   /SetVariables(Particles.PositionAndSizeInfo:Vector4f=&MakeVector4FromVectorAndFloat(
     "Vector (RGB)"=&TransformPosition(
      Position=&AddVectorToPosition(Position=$Particles.Position, Vector=<0 0 0>)
      Source Space=::Simulation
      Destination Space=::World
      Apply Scale=true
     )
     "Float (Alpha)"=&MakeFloatFromVector2d(Vector 2D=$Particles.SpriteSize)
    )) {isSetParameters=true}
   /SetVariables(Particles.CapsuleInfo:Vector4f=&MakeVector4(X=$Particles.CapHeight, Y=$Particles.CapRadius, Z=$Particles.CapScaleTip, W=$Particles.CapsuleOffset)) {isSetParameters=true}
   &self:SecondParticleMotion(
    Age:NiagaraFloat=$Particles.Timeline.Time
    Offset:NiagaraFloat=&Multiply_Float(
     A=&RandomRangeFloat(Minimum=0.6, Maximum=0.9)
     B=&MakeFloatFromVector2d(Vector 2D=$Particles.SpriteSize)
    )
    RandomZ:Vector3f=&NormalizeVector(
     Vector To Normalize=&RandomRangeVector(Minimum=<-1 -1 -1>, Maximum=<1 1 1>)
     Safe Normalize=true
     Safe Fallback Threshold=0
     Safe Fallback Vector=<0 0 0.001>
    )
    Speed:NiagaraFloat=&RandomRangeFloat(Minimum=100, Maximum=150)
   ) {module=SecondParticleMotion}
   /SetVariables(Particles.SecondBlobInfo:Vector4f=&MakeVector4(
     X=&MakeFloatFromVector(
      Vector=&TransformPosition(Position=$Particles.SecondParticleMotion.SecBPosition, Source Space=::Simulation, Destination Space=::World)
      Channel=::NewEnumerator0
     )
     Y=&MakeFloatFromVector(
      Vector=&TransformPosition(Position=$Particles.SecondParticleMotion.SecBPosition, Source Space=::Simulation, Destination Space=::World)
      Channel=::NewEnumerator1
     )
     Z=&MakeFloatFromVector(
      Vector=&TransformPosition(Position=$Particles.SecondParticleMotion.SecBPosition, Source Space=::Simulation, Destination Space=::World)
      Channel=::NewEnumerator2
     )
     W=&Multiply_Float(
      A=&MakeFloatFromVector2d(Vector 2D=$Particles.SpriteSize)
      B=&RandomRangeFloat(Minimum=0.9, Maximum=1.1)
     )
    )) {isSetParameters=true}
   /SetVariables() {isSetParameters=true}
   /SetVariables(Particles.PriteSizeForSoftGlow:Vector2f=&Multiply_Vector2D(A=$Particles.SpriteSize, B=<7 7>)) {isSetParameters=true}
   &ScaleColor(
    Linear Color Curve=DataInterface() {propertyValues {AlphaCurve {keys [<at 0 0>,<at 0.3223 1>]}, BlueCurve {keys [<at 0 1>]}, GreenCurve {keys [<at 0 1>]}, RedCurve {keys [<at 0 1>]}}}
    Scale Mode=2
   )
  ]
  rendererClasses [&NiagaraMeshRendererProperties, &NiagaraSpriteRendererProperties]
  renderers [
   {rendererClass=&NiagaraMeshRendererProperties, rendererIndex=0}
   {rendererClass=&NiagaraMeshRendererProperties, rendererIndex=1}
   {rendererClass=&NiagaraSpriteRendererProperties, rendererIndex=2}
   {rendererClass=&NiagaraMeshRendererProperties, rendererIndex=3}
   {rendererClass=&NiagaraMeshRendererProperties, rendererIndex=4}
  ]
 }
}
summary {
 systemName=NS_ReactFx_BandSoda
 emitters [
  {emitterName=Metaball1, rendererClasses [&NiagaraMeshRendererProperties, &NiagaraSpriteRendererProperties]}
  {emitterName=FallBack, rendererClasses [&NiagaraMeshRendererProperties]}
  {emitterName=Droplets, rendererClasses []}
  {emitterName=FallBackHeart, rendererClasses []}
 ]
 userVariables [
  User.bIsFrontEnd {type=NiagaraBool}
  User.bIsFrontEndPreview {type=NiagaraBool}
  User.ColorIndex {type=NiagaraInt32}
 ]
}
